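; Cosmetic part modifier for the PawJasmineSword character part that makes its idle Niagara VFX react to player kills.
; After res-in in gameplay, it watches the AthenaKills stat, pushes it to KillCount, stamps GameTime when above zero, and debug-prints the value.
; In the front end, a looping timer instead previews the reaction with a random KillCount and current GameTime.
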(event Custom|TriggerReactivity
  (bind _asniagara_particle_system_component (Utilities|Casting|CastToNiagaraComponent (CallFunction|GetIdleVFXComponent)))
  (Niagara|SetNiagaraVariable(Int32) _asniagara_particle_system_component "KillCount" (Math|Random|RandomInteger 999))
  (Effects|Components|ParticleSystem|SetFloatParameter _asniagara_particle_system_component "GameTime" (Utilities|Time|GetGameTimeinSeconds)))

(event EventOnPostResIn
  (bind _returnvalue (CallFunction|GetIdleVFXComponent))
  (|CallParentFunction)
  (if (CustomCosmeticModifier|IsinFrontEnd)
    (Utilities|Time|SetTimerbyEvent (delegate Custom|TriggerReactivity) 9.0 true false -6.0)
    (else
      (bind (_statname _statvalue) (Cosmetics|WatchCosmeticStat (CustomCharacter|GetAssociatedPlayerPawn) "(TagName=\"StatManager.AthenaKills\")")
        (:OnStatChanged
          (Development|PrintString (Utilities|String|ToString(Integer) _statvalue))
          (bind _asniagara_particle_system_component (Utilities|Casting|CastToNiagaraComponent _returnvalue))
          (Niagara|SetNiagaraVariable(Int32) _asniagara_particle_system_component "KillCount" _statvalue)
          (if (> _statvalue 0)
            (Effects|Components|ParticleSystem|SetFloatParameter _returnvalue "GameTime" (Utilities|Time|GetGameTimeinSeconds))))))))

(fn ConstructionScript
  (|CallParentFunction))
