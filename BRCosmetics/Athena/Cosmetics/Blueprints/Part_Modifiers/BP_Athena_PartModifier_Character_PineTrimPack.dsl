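; Character part modifier for the PineTrimPack cosmetic that makes its idle VFX and materials react to the player's eliminations.
; In-game it watches the AthenaKills stat and increments its local kill count when the stat exceeds it.
; In the front end, a delayed first trigger and a looping timer simulate kills.
; Each reaction sends the count to the particle ReactivityCounter and stamps game time into the mesh materials' BP_GetTime parameter.
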(event Custom|FrontEndReactivity
  ; Front End Loop
  (Variables|Default|SetKillCount (+ (Variables|Default|GetKillCount) 1))
  ; FX
  (Effects|Components|ParticleSystem|SetIntParameter (CallFunction|GetIdleVFXComponent) "ReactivityCounter" (Variables|Default|GetKillCount))
  (Rendering|Material|SetScalarParameterValueonMaterials (CustomCharacter|GetSkeletalMeshForPartOnPlayerPawn) "BP_GetTime" (Utilities|Time|GetGameTimeinSeconds)))

(event EventOnPostResIn
  (|CallParentFunction)
  (if (CustomCosmeticModifier|IsinFrontEnd)
    ; Front End Loop
    (Utilities|Time|SetTimerbyEvent (delegate Custom|FrontEndReactivity) 1e+01 true)
    (Utilities|FlowControl|Delay 2.0)
    (else
      (bind (_statname _statvalue) (Cosmetics|WatchCosmeticStat (CustomCharacter|GetAssociatedPlayerPawn) "(TagName=\"StatManager.AthenaKills\")" 0 false)
        (:OnStatChanged
          (if (and (> _statvalue 0) (> _statvalue (Variables|Default|GetKillCount)))
            _)))))
  (Variables|Default|SetKillCount (+ (Variables|Default|GetKillCount) 1))
  ; FX
  (Effects|Components|ParticleSystem|SetIntParameter (CallFunction|GetIdleVFXComponent) "ReactivityCounter" (Variables|Default|GetKillCount))
  (Rendering|Material|SetScalarParameterValueonMaterials (CustomCharacter|GetSkeletalMeshForPartOnPlayerPawn) "BP_GetTime" (Utilities|Time|GetGameTimeinSeconds)))

(fn ConstructionScript
  (|CallParentFunction))
