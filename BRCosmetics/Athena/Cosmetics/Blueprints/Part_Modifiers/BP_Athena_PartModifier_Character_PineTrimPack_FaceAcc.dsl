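; Character part modifier for the PineTrimPack face accessory, feeding pose and timing data into its materials for reactive effects.
; Each tick sends the head socket position and its offset from neck_02 as vector parameters.
; After res-in it enables BP_UseCCPM and writes game time to BP_GetTime.
; A looping timer refreshes that time in the front end; otherwise an AthenaKills stat watch increments KillCount.
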(event EventTick
  (params (DeltaSeconds :type Float%20%28single-precision%29))
  (bind _returnvalue (CustomCharacter|GetSkeletalMeshForPartOnPlayerPawn))
  (bind _returnvalue_2 (Transformation|GetSocketLocation _returnvalue "head"))
  ; Send Socket Location to the Material
  (|CallParentFunction DeltaSeconds)
  (Rendering|Material|SetVectorParameterValueonMaterials _returnvalue "BP_OffsetVector_Head" (- _returnvalue_2 (Transformation|GetSocketLocation _returnvalue "neck_02")))
  (Rendering|Material|SetVectorParameterValueonMaterials _returnvalue "BP_HeadPosition" _returnvalue_2))

(event Custom|FrontEndReactivity
  ; Front End Loop
  (Rendering|Material|SetScalarParameterValueonMaterials (CustomCharacter|GetSkeletalMeshForPartOnPlayerPawn) "BP_GetTime" (Utilities|Time|GetGameTimeinSeconds)))

(event EventOnPostResIn
  (|CallParentFunction)
  (Rendering|Material|SetScalarParameterValueonMaterials (CustomCharacter|GetSkeletalMeshForPartOnPlayerPawn) "BP_UseCCPM" 1.0)
  (if (CustomCosmeticModifier|IsinFrontEnd)
    ; Front End Loop
    (Utilities|Time|SetTimerbyEvent (delegate Custom|FrontEndReactivity) 1e+01 true)
    (Utilities|FlowControl|Delay 2.0)
    (else
      (bind (_statname _statvalue) (Cosmetics|WatchCosmeticStat (CustomCharacter|GetAssociatedPlayerPawn) "(TagName=\"StatManager.AthenaKills\")" 0 false)
        (:OnStatChanged
          (if (and (> _statvalue 0) (> _statvalue (Variables|Default|GetKillCount)))
            (Variables|Default|SetKillCount (+ (Variables|Default|GetKillCount) 1)))))))
  (Rendering|Material|SetScalarParameterValueonMaterials (CustomCharacter|GetSkeletalMeshForPartOnPlayerPawn) "BP_GetTime" (Utilities|Time|GetGameTimeinSeconds)))

(fn ConstructionScript
  (|CallParentFunction))
