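; Cosmetic modifier Blueprint for the BaskIsle back bling, adding reactive effects that appear tied to a chest count.
; Its Trigger Reactivity FX event increments the count, passes it to the idle VFX as ReactivityChestCount, and timestamps the skeletal mesh materials with game time.
; After res-in it stores the idle Niagara component and, in the front end only, triggers the reaction once after a short delay.
; Nothing here shows what triggers the reaction during gameplay.
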(event Custom|Trigger%20Reactivity%20FX
  (Utilities|FlowControl|Sequence
    (:then_0
      (bind _result (Math|Integer|IncrementInt (Variables|Default|GetChestCount))
        (:%20%20
          (Effects|Components|ParticleSystem|SetIntParameter (CallFunction|GetIdleVFXComponent) "ReactivityChestCount" _result))))
    (:then_1
      ; Material Reactivity
      (Rendering|Material|SetScalarParameterValueonMaterials (CustomCharacter|GetSkeletalMeshForPartOnPlayerPawn) "Timestamp" (Utilities|Time|GetGameTimeinSeconds)))))

(event EventOnPostResIn
  (|CallParentFunction)
  (bind _asniagara_particle_system_component (Utilities|Casting|CastToNiagaraComponent (CallFunction|GetIdleVFXComponent)))
  (Variables|Default|SetNiagara%20System _asniagara_particle_system_component)
  (if (CustomCosmeticModifier|IsinFrontEnd)
    ; FE Reactivity
    (Utilities|FlowControl|Delay 3.0)
    (CallFunction|TriggerReactivityFX)
    (Utilities|FlowControl|Delay 15.66)))

(fn ConstructionScript
  (|CallParentFunction))
