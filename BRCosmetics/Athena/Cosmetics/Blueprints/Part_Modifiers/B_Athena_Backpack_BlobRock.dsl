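; Reactive BlobRock back bling part modifier that plays screen animations for kills, health gains, shield gains and chest opens.
; Watched stats trigger a transition, activation, then timed return to idle, swapping flipbook textures and timestamps on a dynamic material alongside particle LightID.
; Kills bypass the readiness gate; the other three share one animation and sound.
; The frontend alternates kill and health reactions on a looping timer, without audio.
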(event Custom|FrontEnd%20Cycle
  (switch int (Variables|Default|GetFrontEndInt)
    (:0
      (CallFunction|Kills)
      (Variables|Default|SetFrontEndInt 1)
      (Variables|Default|SetReadyForNewReaction true))
    (:1
      (CallFunction|GainedHealth)
      (Variables|Default|SetFrontEndInt 0)
      (Variables|Default|SetReadyForNewReaction true))))

(event EventOnPostResIn
  (|CallParentFunction)
  (bind _returnvalue (Rendering|Material|CreateDynamicMaterialInstance (CustomCharacter|GetSkeletalMeshForPartOnPlayerPawn) 1 "/BRCosmetics/Accessories/FORT_Backpacks/Backpack_M_MED_BlobRock/Materials/MI_Backpack_BlobRock_Screen.MI_Backpack_BlobRock_Screen" "DMI_BlobRockScreen" :!class /Script/Engine.PrimitiveComponent))
  (Variables|Default|SetBlobRockScreen _returnvalue)
  (if (Reactive|Frontend|IsCosmeticObjectDisplayedinFrontendContext 0)
    (Utilities|Time|SetTimerbyEvent (EventDispatchers|CreateEvent :!event Custom|FrontEnd%20Cycle) 13.0 true)
    (else
      (Utilities|FlowControl|Sequence
        (:then_0
          (bind (_statname _statvalue) (Cosmetics|WatchCosmeticStat (CustomCharacter|GetAssociatedPlayerPawn) "(TagName=\"StatManager.AthenaKills\")" 0 false)
            (:OnStatChanged
              (if (> _statvalue 0)
                (Variables|Default|SetIsKilled true)
                (CallFunction|Kills)))))
        (:then_1
          (bind (_statname_2 _statvalue_2) (Cosmetics|WatchCosmeticStat (CustomCharacter|GetAssociatedPlayerPawn) "(TagName=\"StatManager.GainedHealthTimes\")" 0 false -1.0 -1.0 false)
            (:OnStatChanged
              (if (> _statvalue_2 0)
                (CallFunction|GainedHealth)))))
        (:then_2
          (bind (_statname_3 _statvalue_3) (Cosmetics|WatchCosmeticStat (CustomCharacter|GetAssociatedPlayerPawn) "(TagName=\"StatManager.GainedShieldTimes\")" 0 false -1.0 -1.0 false)
            (:OnStatChanged
              (if (> _statvalue_3 0)
                (CallFunction|GainedShield)))))
        (:then_3
          (bind (_statname_4 _statvalue_4) (Cosmetics|WatchCosmeticStat (CustomCharacter|GetAssociatedPlayerPawn) "(TagName=\"StatManager.OpenChests\")" 0 false 4.0 -1.0 false)
            (:OnStatChanged
              (if (> _statvalue_4 0)
                (CallFunction|OpenChests)))))))))

(event Custom|GainedHealth
  (if (Variables|Default|GetReadyForNewReaction)
    (Variables|Default|SetReadyForNewReaction false)
    (CallFunction|SetTransitionParameters)
    (Effects|Components|ParticleSystem|SetIntParameter (CallFunction|GetIdleVFXComponent) "LightID" 1)
    ; let transition play for a bit
    (Utilities|FlowControl|Delay (Variables|Default|GetTransitionTimeLength))
    ; toggle activation flipbooks
    (Variables|Default|SetActivationFlipbookNumber 3.0)
    (CallFunction|SetActivation2Parameters)
    (Effects|Components|ParticleSystem|SetIntParameter (CallFunction|GetIdleVFXComponent) "LightID" 3)
    ; play audio
    (CallFunction|AudioTrigger "(TagName=\"SoundLibrary.Cosmetic.OpenChests\")")
    (Utilities|Time|SetTimerbyEvent (EventDispatchers|CreateEvent :!event Custom|Reset%20to%20Default) (+ (Variables|Default|GetActivationLength) (Variables|Default|GetHoldLastFrameTime)))))

(event Custom|GainedShield
  (if (Variables|Default|GetReadyForNewReaction)
    (Variables|Default|SetReadyForNewReaction false)
    (CallFunction|SetTransitionParameters)
    (Effects|Components|ParticleSystem|SetIntParameter (CallFunction|GetIdleVFXComponent) "LightID" 1)
    ; let transition play for a bit
    (Utilities|FlowControl|Delay (Variables|Default|GetTransitionTimeLength))
    ; toggle activation flipbooks
    (Variables|Default|SetActivationFlipbookNumber 3.0)
    (CallFunction|SetActivation2Parameters)
    (Effects|Components|ParticleSystem|SetIntParameter (CallFunction|GetIdleVFXComponent) "LightID" 3)
    ; play audio
    (CallFunction|AudioTrigger "(TagName=\"SoundLibrary.Cosmetic.OpenChests\")")
    (Utilities|Time|SetTimerbyEvent (EventDispatchers|CreateEvent :!event Custom|Reset%20to%20Default) (+ (Variables|Default|GetActivationLength) (Variables|Default|GetHoldLastFrameTime)))))

(event Custom|OpenChests
  (if (Variables|Default|GetReadyForNewReaction)
    (Variables|Default|SetReadyForNewReaction false)
    (CallFunction|SetTransitionParameters)
    (Effects|Components|ParticleSystem|SetIntParameter (CallFunction|GetIdleVFXComponent) "LightID" 1)
    ; let transition play for a bit
    (Utilities|FlowControl|Delay (Variables|Default|GetTransitionTimeLength))
    ; toggle activation flipbooks
    (Variables|Default|SetActivationFlipbookNumber 3.0)
    (CallFunction|SetActivation2Parameters)
    (Effects|Components|ParticleSystem|SetIntParameter (CallFunction|GetIdleVFXComponent) "LightID" 3)
    ; play audio
    (CallFunction|AudioTrigger "(TagName=\"SoundLibrary.Cosmetic.OpenChests\")")
    (Utilities|Time|SetTimerbyEvent (EventDispatchers|CreateEvent :!event Custom|Reset%20to%20Default) (+ (Variables|Default|GetActivationLength) (Variables|Default|GetHoldLastFrameTime)))))

(event Custom|Kills
  (Variables|Default|SetReadyForNewReaction false)
  (CallFunction|SetTransitionParameters)
  (Effects|Components|ParticleSystem|SetIntParameter (CallFunction|GetIdleVFXComponent) "LightID" 1)
  ; let transition play for a bit
  (Utilities|FlowControl|Delay (Variables|Default|GetTransitionTimeLength))
  ; toggle activation flipbooks
  (Variables|Default|SetActivationFlipbookNumber 2.0)
  (CallFunction|SetActivation1Parameters)
  (Effects|Components|ParticleSystem|SetIntParameter (CallFunction|GetIdleVFXComponent) "LightID" 2)
  ; play audio kill sound
  (CallFunction|AudioTrigger "(TagName=\"SoundLibrary.Cosmetic.Kills\")")
  (Utilities|Time|SetTimerbyEvent (EventDispatchers|CreateEvent :!event Custom|Reset%20to%20Default) (+ (Variables|Default|GetActivationLength) (Variables|Default|GetHoldLastFrameTime))))

(event Custom|Reset%20to%20Default
  (Variables|Default|SetIsKilled false)
  (CallFunction|SetTransitionParameters)
  (Effects|Components|ParticleSystem|SetIntParameter (CallFunction|GetIdleVFXComponent) "LightID" 1)
  ; let transition play for a bit
  (Utilities|FlowControl|Delay (Variables|Default|GetTransitionTimeLength))
  (CallFunction|SetIdleParameters)
  (Effects|Components|ParticleSystem|SetIntParameter (CallFunction|GetIdleVFXComponent) "LightID" 0)
  (Utilities|FlowControl|RetriggerableDelay (+ (Variables|Default|GetActivationLength) (Variables|Default|GetHoldLastFrameTime)))
  (Variables|Default|SetReadyForNewReaction true))

(event Custom|DamageDealtAudio
  (CallFunction|AudioTrigger "(TagName=\"SoundLibrary.Cosmetic.DamageDealt\")")
  (Utilities|Time|SetTimerbyEvent (EventDispatchers|CreateEvent :!event Custom|DamageDealtReset) (Math|Random|RandomFloatinRange (Variables|Default|GetDamageReceivedLowRandomDelay) (Variables|Default|GetDamageReceivedHighRandomDelay))))

(event Custom|DamageReceivedAudio
  (CallFunction|AudioTrigger "(TagName=\"SoundLibrary.Cosmetic.DamageReceived\")")
  (Utilities|Time|SetTimerbyEvent (EventDispatchers|CreateEvent :!event Custom|DamageReceivedReset) (Math|Random|RandomFloatinRange (Variables|Default|GetDamageReceivedLowRandomDelay) (Variables|Default|GetDamageDealtHighRandomDelay))))

(event Custom|DamageDealtReset
  (Variables|Default|SetDamageDealtAudioEnabled true))

(event Custom|DamageReceivedReset
  (Variables|Default|SetDamageReceivedAudioEnabled true))

(fn AudioTrigger
  (params (SoundLibTag :type Sound%20Library%20Tag%20Structure))
  (bind _returnvalue (CustomCharacter|GetAssociatedPlayerPawn))
  (if (and (Pawn|IsLocallyViewed _returnvalue) (not (Reactive|Frontend|IsCosmeticObjectDisplayedinFrontendContext 0)))
    (bind _returnvalue_2 (SoundLibrary|PlaySoundSimple (WorldSubsystems|GetSoundLibrarySubsystem) _returnvalue (Utilities|Struct|MakeSimpleSoundSettings SoundLibTag true false true false 1.0 1.0 0.0 0 0 "None" (CustomCharacter|GetSkeletalMeshForPartOnPlayerPawn)) :!class /Script/SoundLibrary.SoundLibrarySubsystem))
    (return (.bSuccess (Utilities|Struct|BreakSoundLibraryPlayResult _returnvalue_2)))))

(fn DamageDealt)

(fn DamageReceived)

(fn SetActivation1Parameters
  ; switch to transition flipbook
  (Rendering|Material|SetScalarParameterValue (Variables|Default|GetBlobRockScreen) "BlobRockFlipbook" (Variables|Default|GetActivationFlipbookNumber) :!class /Script/Engine.MaterialInstanceDynamic)
  (Rendering|Material|SetTextureParameterValue (Variables|Default|GetBlobRockScreen) "BlobRockFlipbookTexture" (Variables|Default|GetActivation1Texture))
  (Rendering|Material|SetScalarParameterValue (Variables|Default|GetBlobRockScreen) "FlipbookRowSize" 6.0 :!class /Script/Engine.MaterialInstanceDynamic)
  (Rendering|Material|SetScalarParameterValue (Variables|Default|GetBlobRockScreen) "FlipbookColumnSize" 4.0 :!class /Script/Engine.MaterialInstanceDynamic)
  (Rendering|Material|SetScalarParameterValue (Variables|Default|GetBlobRockScreen) "Timestamp" (Utilities|Time|GetGameTimeinSeconds) :!class /Script/Engine.MaterialInstanceDynamic)
  (Rendering|Material|SetScalarParameterValue (Variables|Default|GetBlobRockScreen) "TransitionLength" (Variables|Default|GetActivationLength) :!class /Script/Engine.MaterialInstanceDynamic)
  (Rendering|Material|SetScalarParameterValue (Variables|Default|GetBlobRockScreen) "HoldLastFrameTime" (Variables|Default|GetHoldLastFrameTime) :!class /Script/Engine.MaterialInstanceDynamic))

(fn SetActivation2Parameters
  ; switch to transition flipbook
  (Rendering|Material|SetScalarParameterValue (Variables|Default|GetBlobRockScreen) "BlobRockFlipbook" (Variables|Default|GetActivationFlipbookNumber) :!class /Script/Engine.MaterialInstanceDynamic)
  (Rendering|Material|SetTextureParameterValue (Variables|Default|GetBlobRockScreen) "BlobRockFlipbookTexture" (Variables|Default|GetActivation2Texture))
  (Rendering|Material|SetScalarParameterValue (Variables|Default|GetBlobRockScreen) "FlipbookRowSize" 7.0 :!class /Script/Engine.MaterialInstanceDynamic)
  (Rendering|Material|SetScalarParameterValue (Variables|Default|GetBlobRockScreen) "FlipbookColumnSize" 4.0 :!class /Script/Engine.MaterialInstanceDynamic)
  (Rendering|Material|SetScalarParameterValue (Variables|Default|GetBlobRockScreen) "Timestamp" (Utilities|Time|GetGameTimeinSeconds) :!class /Script/Engine.MaterialInstanceDynamic)
  (Rendering|Material|SetScalarParameterValue (Variables|Default|GetBlobRockScreen) "TransitionLength" (Variables|Default|GetActivationLength) :!class /Script/Engine.MaterialInstanceDynamic)
  (Rendering|Material|SetScalarParameterValue (Variables|Default|GetBlobRockScreen) "HoldLastFrameTime" (Variables|Default|GetHoldLastFrameTime) :!class /Script/Engine.MaterialInstanceDynamic))

(fn SetIdleParameters
  ; switch to transition flipbook, doesn't need timelength set because it loops in the material regardless of timelength value
  (Rendering|Material|SetScalarParameterValue (Variables|Default|GetBlobRockScreen) "BlobRockFlipbook" 0.0 :!class /Script/Engine.MaterialInstanceDynamic)
  (Rendering|Material|SetTextureParameterValue (Variables|Default|GetBlobRockScreen) "BlobRockFlipbookTexture" (Variables|Default|GetIdleTexture))
  (Rendering|Material|SetScalarParameterValue (Variables|Default|GetBlobRockScreen) "FlipbookRowSize" 4.0 :!class /Script/Engine.MaterialInstanceDynamic)
  (Rendering|Material|SetScalarParameterValue (Variables|Default|GetBlobRockScreen) "FlipbookColumnSize" 3.0 :!class /Script/Engine.MaterialInstanceDynamic))

(fn SetTransitionParameters
  ; switch to transition flipbook
  (Rendering|Material|SetScalarParameterValue (Variables|Default|GetBlobRockScreen) "BlobRockFlipbook" 1.0 :!class /Script/Engine.MaterialInstanceDynamic)
  (Rendering|Material|SetTextureParameterValue (Variables|Default|GetBlobRockScreen) "BlobRockFlipbookTexture" (Variables|Default|GetTransitionTexture))
  (Rendering|Material|SetScalarParameterValue (Variables|Default|GetBlobRockScreen) "FlipbookRowSize" 2.0 :!class /Script/Engine.MaterialInstanceDynamic)
  (Rendering|Material|SetScalarParameterValue (Variables|Default|GetBlobRockScreen) "FlipbookColumnSize" 2.0 :!class /Script/Engine.MaterialInstanceDynamic)
  (Rendering|Material|SetScalarParameterValue (Variables|Default|GetBlobRockScreen) "Timestamp" (Utilities|Time|GetGameTimeinSeconds) :!class /Script/Engine.MaterialInstanceDynamic)
  (Rendering|Material|SetScalarParameterValue (Variables|Default|GetBlobRockScreen) "TransitionLength" (Variables|Default|GetTransitionTimeLength) :!class /Script/Engine.MaterialInstanceDynamic))

(fn ConstructionScript
  (|CallParentFunction))
