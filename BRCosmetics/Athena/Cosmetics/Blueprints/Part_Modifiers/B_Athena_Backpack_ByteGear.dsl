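; Reactive ByteGear back bling part modifier that changes a face and plays sounds on kills, damage dealt and damage received.
; Cosmetic stat watchers trigger events that set the FaceFrame material parameter, with a timer restoring the default face.
; A kill flag suppresses the damage-dealt face, and randomized cooldowns throttle damage audio.
; Frontend display cycles the reactions on a looping timer, while sound plays only for the locally viewed pawn outside frontend.
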(event Custom|Kills
  (Rendering|Material|SetScalarParameterValueonMaterials (CustomCharacter|GetSkeletalMeshForPartOnPlayerPawn) "FaceFrame" 4.0)
  (CallFunction|AudioTrigger "(TagName=\"SoundLibrary.Cosmetic.Kills\")")
  (Utilities|Time|SetTimerbyEvent (EventDispatchers|CreateEvent :!event Custom|Reset%20to%20Default) 4.0))

(event Custom|FrontEnd%20Cycle
  (switch int (Variables|Default|GetFrontEndInt)
    (:0
      (CallFunction|Kills))
    (:1
      (CallFunction|DamageDealt))
    (:2
      (CallFunction|DamageReceived)
      (Variables|Default|SetFrontEndInt 0)))
  (Variables|Default|SetFrontEndInt (+ (Variables|Default|GetFrontEndInt) 1)))

(event Custom|DamageDealt
  (Variables|Default|GetDamageDealtAudioEnabled
    (:then
      (CallFunction|DamageDealtAudio)
      (Variables|Default|SetDamageDealtAudioEnabled false))
    (:else
      _))
  (Variables|Default|GetIsKilled
    (:then
      _)
    (:else
      (Rendering|Material|SetScalarParameterValueonMaterials (CustomCharacter|GetSkeletalMeshForPartOnPlayerPawn) "FaceFrame" 3.0)))
  (Utilities|Time|SetTimerbyEvent (EventDispatchers|CreateEvent :!event Custom|Reset%20to%20Default) 4.0))

(event EventOnPostResIn
  (|CallParentFunction)
  (if (Reactive|Frontend|ShouldDisplayFrontendReactivity 0 "Backpack")
    (Utilities|Time|SetTimerbyEvent (EventDispatchers|CreateEvent :!event Custom|FrontEnd%20Cycle) 6.0 true)
    (else
      (Utilities|FlowControl|Sequence
        (:then_0
          (bind (_statname _statvalue) (Cosmetics|WatchCosmeticStat (CustomCharacter|GetAssociatedPlayerPawn) "(TagName=\"StatManager.AthenaKills\")" 0 false)
            (:OnStatChanged
              (if (> _statvalue 0)
                (Variables|Default|SetIsKilled true)
                (CallFunction|Kills)))))
        (:then_1
          (bind (_statname_2 _statvalue_2) (Cosmetics|WatchCosmeticStat (CustomCharacter|GetAssociatedPlayerPawn) "(TagName=\"StatManager.DamageDealt\")" 0 false)
            (:OnStatChanged
              (if (> _statvalue_2 0)
                (CallFunction|DamageDealt)))))
        (:then_2
          (bind (_statname_3 _statvalue_3) (Cosmetics|WatchCosmeticStat (CustomCharacter|GetAssociatedPlayerPawn) "(TagName=\"StatManager.DamageReceived\")" 0 false)
            (:OnStatChanged
              (if (> _statvalue_3 0)
                (CallFunction|DamageReceived)))))))))

(event Custom|DamageReceived
  (Rendering|Material|SetScalarParameterValueonMaterials (CustomCharacter|GetSkeletalMeshForPartOnPlayerPawn) "Timestamp" (Utilities|Time|GetGameTimeinSeconds))
  (Rendering|Material|SetScalarParameterValueonMaterials (CustomCharacter|GetSkeletalMeshForPartOnPlayerPawn) "TransitionLength" (Variables|Default|GetTransition%20Length))
  (Rendering|Material|SetScalarParameterValueonMaterials (CustomCharacter|GetSkeletalMeshForPartOnPlayerPawn) "FaceFrame" 2.0)
  (if (Variables|Default|GetDamageReceivedAudioEnabled)
    (CallFunction|DamageReceivedAudio)
    (Variables|Default|SetDamageReceivedAudioEnabled false))
  (Utilities|Time|SetTimerbyEvent (EventDispatchers|CreateEvent :!event Custom|Reset%20to%20Default) 4.0))

(event Custom|Reset%20to%20Default
  (Variables|Default|SetIsKilled false)
  (Rendering|Material|SetScalarParameterValueonMaterials (CustomCharacter|GetSkeletalMeshForPartOnPlayerPawn) "FaceFrame" 0.0))

(event Custom|DamageDealtAudio
  (CallFunction|AudioTrigger "(TagName=\"SoundLibrary.Cosmetic.DamageDealt\")")
  (Utilities|Time|SetTimerbyEvent (EventDispatchers|CreateEvent :!event Custom|DamageDealtReset) (Math|Random|RandomFloatinRange (Variables|Default|GetDamageReceivedLowRandomDelay) (Variables|Default|GetDamageReceivedHighRandomDelay))))

(event Custom|DamageReceivedAudio
  (CallFunction|AudioTrigger "(TagName=\"SoundLibrary.Cosmetic.DamageReceived\")")
  (Utilities|Time|SetTimerbyEvent (EventDispatchers|CreateEvent :!event Custom|DamageReceivedReset) (Math|Random|RandomFloatinRange (Variables|Default|GetDamageReceivedLowRandomDelay) (Variables|Default|GetDamageDealtHighRandomDelay))))

(event Custom|DamageDealtReset
  (Variables|Default|SetDamageDealtAudioEnabled true))

(event Custom|DamageReceivedReset
  (Variables|Default|SetDamageReceivedAudioEnabled true))

(fn AudioTrigger
  (params (SoundLibTag :type Sound%20Library%20Tag%20Structure))
  (bind _returnvalue (CustomCharacter|GetAssociatedPlayerPawn))
  (if (and (Pawn|IsLocallyViewed _returnvalue) (not (Reactive|Frontend|IsCosmeticObjectDisplayedinFrontendContext 0)))
    (bind _returnvalue_2 (SoundLibrary|PlaySoundSimple (WorldSubsystems|GetSoundLibrarySubsystem) _returnvalue (Utilities|Struct|MakeSimpleSoundSettings SoundLibTag true false true false 1.0 1.0 0.0 0 0 "None" (CustomCharacter|GetSkeletalMeshForPartOnPlayerPawn)) :!class /Script/SoundLibrary.SoundLibrarySubsystem))
    (return (.bSuccess (Utilities|Struct|BreakSoundLibraryPlayResult _returnvalue_2)))))

(fn ConstructionScript
  (|CallParentFunction))
