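; Backpack part modifier for the CyberDelivery cosmetic that shows the owner's kill count as two digits through material parameters.
; After res-in, gameplay watches the Athena kills stat and accepts only increases up to 99.
; Frontend display instead flags the material and increments a preview count after a delay.
; Both paths split the count into tens and units and pass game time to the material.
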(event EventOnPostResIn
  (bind _returnvalue (CustomCharacter|GetSkeletalMeshForPartOnPlayerPawn))
  (bind _returnvalue_2 (CustomCharacter|GetSkeletalMeshForPartOnPlayerPawn))
  (|CallParentFunction)
  (Utilities|FlowControl|Sequence
    (:then_0
      (if (Reactive|Frontend|IsCosmeticObjectDisplayedinFrontendContext 0)
        ; Front End Loop
        (Rendering|Material|SetScalarParameterValueonMaterials (CustomCharacter|GetSkeletalMeshForPartOnPlayerPawn) "BP_IsFrontEnd" 1.0)
        (Variables|Default|SetFrontEndKillCount 0)
        (Utilities|FlowControl|Delay 4.0)
        (bind _output_get (Variables|Default|SetFrontEndKillCount (+ (Variables|Default|GetFrontEndKillCount) 1)))
        (if (<= _output_get 99)
          (Rendering|Material|SetScalarParameterValueonMaterials _returnvalue "BP_FirstDigit" (Math|Conversions|ToFloat(Integer) (Math|Float|Floor (/ (Math|Conversions|ToFloat(Integer) _output_get) 1e+01))))
          (Rendering|Material|SetScalarParameterValueonMaterials _returnvalue "BP_SecondDigit" (- (Math|Conversions|ToFloat(Integer) _output_get) (* (Math|Conversions|ToFloat(Integer) (Math|Float|Floor (/ (Math|Conversions|ToFloat(Integer) _output_get) 1e+01))) 1e+01)))
          (Rendering|Material|SetScalarParameterValueonMaterials _returnvalue "BP_GetTime" (Utilities|Time|GetGameTimeinSeconds))
          (Utilities|FlowControl|Delay 12.0))
        (else
          ;[1 0.7262 0.2176 1 60] On Kill Reactivity ☠
          (bind (_statname _statvalue) (Cosmetics|WatchCosmeticStat (CustomCharacter|GetAssociatedPlayerPawn) "(TagName=\"StatManager.AthenaKills\")" 0 false)
            (:OnStatChanged
              (if (and (> _statvalue (Variables|Default|GetKillCount)) (<= _statvalue 99))
                ;;[1 1 1] Material Reactivity
                (Rendering|Material|SetScalarParameterValueonMaterials _returnvalue_2 "BP_FirstDigit" (Math|Conversions|ToFloat(Integer) (Math|Float|Floor (/ (Math|Conversions|ToFloat(Integer) _statvalue) 1e+01))))
                (Rendering|Material|SetScalarParameterValueonMaterials _returnvalue_2 "BP_SecondDigit" (- (Math|Conversions|ToFloat(Integer) _statvalue) (* (Math|Conversions|ToFloat(Integer) (Math|Float|Floor (/ (Math|Conversions|ToFloat(Integer) _statvalue) 1e+01))) 1e+01)))
                (Rendering|Material|SetScalarParameterValueonMaterials _returnvalue_2 "BP_GetTime" (Utilities|Time|GetGameTimeinSeconds))
                (Variables|Default|SetKillCount _statvalue)))))))))

(fn ConstructionScript
  (|CallParentFunction))
