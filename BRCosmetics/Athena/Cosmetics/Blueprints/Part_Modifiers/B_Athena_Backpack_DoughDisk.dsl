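; Backpack part modifier for the DoughDisk cosmetic that plays a Niagara transition effect when a dual-wield pickaxe is put away.
; After post-resource init it binds FXTrigger to the backpack AnimBP's FX event, while weapon changes cache the previous weapon.
; When the trigger reports no pickaxe equipped, a short retriggerable delay runs, then the effect spawns attached via the idle VFX component if that previous weapon casts to the generic dual-wield pickaxe.
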(event EventOnPostResIn
  (|CallParentFunction)
  (bind _asf_med_dough_disk_pack_anim_bp (Utilities|Casting|CastToF_MED_DoughDisk_Pack_AnimBP (Components|SkeletalMesh|GetAnimInstance (Customization|GetSkeletalMeshforPartType (CustomCharacter|GetAssociatedPlayerPawn) (CustomCharacter|GetCharacterPartType) :!class /Script/FortniteGame.FortPlayerPawn))))
  (Default|BindEventtoFXBAckpack _asf_med_dough_disk_pack_anim_bp (delegate Custom|FXTrigger)))

(event Custom|FXTrigger
  (params (EquipedPickaxe :type Boolean))
  (bind (_asb_athena_pickaxe_dual_wield_generic _bsuccess) (Utilities|Casting|CastToB_Athena_Pickaxe_DualWield_Generic (Variables|Default|GetOld%20Weapon)))
  (if (not EquipedPickaxe)
    (Utilities|FlowControl|RetriggerableDelay 0.2)
    (if _bsuccess
      (PartModifier|SpawnSystem|SpawnSystemAttachedPartModifier "/BRCosmetics/Accessories/FORT_Backpacks/Backpack_F_MED_DoughDisk/FX/NS_Backback_DoughDisk_Transition.NS_Backback_DoughDisk_Transition" (CallFunction|GetIdleVFXComponent)))))

(event EventOnWeaponChanged
  (params (NewWeapon :type Fort%20Weapon%20Object%20Reference) (OldWeapon :type Fort%20Weapon%20Object%20Reference))
  (|CallParentFunction NewWeapon OldWeapon)
  (Variables|Default|SetOld%20Weapon OldWeapon))

(fn ConstructionScript
  (|CallParentFunction))
