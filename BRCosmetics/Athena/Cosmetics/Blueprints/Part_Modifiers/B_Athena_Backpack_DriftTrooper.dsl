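; Cosmetic part modifier for the DriftTrooper backpack that makes its lightning material and idle VFX react to the equipped weapon type.
; After res-in, it caches the pawn and backpack mesh, enables LightningReactive on the VFX, and in the front end animates CCPM_Lightning by timeline.
; Weapon-type changes switch both lightning parameters on for harvesting tools and off otherwise, stamping WeaponSwitchTime with game time on the dynamic material.
; The remaining events are empty or disabled stubs.
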(event EventOnPostResIn
  (|CallParentFunction)
  (bind _output_get (Variables|Default|SetPlayerPawn (CustomCharacter|GetAssociatedPlayerPawn)))
  (Variables|Default|SetBodyMesh (Customization|GetSkeletalMeshforPartType _output_get "Backpack" :!class /Script/FortniteGame.FortPlayerPawn))
  (Effects|Components|ParticleSystem|SetFloatParameter (CallFunction|GetIdleVFXComponent) "LightningReactive" 1.0)
  (if (CustomCosmeticModifier|IsinFrontEnd)
    (bind (_direction _stage01) (Timeline|LightningPhase Play)
      (:Update
        (Rendering|Material|SetScalarParameterValue (Class|BTechArtGlobalLibrary|GetDynamicMaterialInstance (Variables|Default|GetBodyMesh)) "CCPM_Lightning" _stage01 :!class /Script/Engine.MaterialInstanceDynamic)))))

(event EventOnNewPlayerWeaponType
  (params (NewType :type EFortItemType%20Enum :default "WorldItem") (OldType :type EFortItemType%20Enum :default "WorldItem"))
  (bind _returnvalue (Utilities|Time|GetGameTimeinSeconds))
  (bind _bodymesh (Variables|Default|GetBodyMesh))
  (bind _material_instance_dynamic (Class|BTechArtGlobalLibrary|GetDynamicMaterialInstance _bodymesh))
  (bind _returnvalue_2 (CallFunction|GetIdleVFXComponent))
  (|CallParentFunction NewType OldType)
  (if (Utilities|Enum|Equal(Enum) NewType "WeaponHarvest")
    (Rendering|Material|SetScalarParameterValue _material_instance_dynamic "CCPM_Lightning" 1.0 :!class /Script/Engine.MaterialInstanceDynamic)
    (Rendering|Material|SetScalarParameterValue _material_instance_dynamic "WeaponSwitchTime" _returnvalue :!class /Script/Engine.MaterialInstanceDynamic)
    (Effects|Components|ParticleSystem|SetFloatParameter _returnvalue_2 "LightningReactive" 1.0)
    (else
      (Rendering|Material|SetScalarParameterValue _material_instance_dynamic "CCPM_Lightning" 0.0 :!class /Script/Engine.MaterialInstanceDynamic)
      (Rendering|Material|SetScalarParameterValue _material_instance_dynamic "WeaponSwitchTime" _returnvalue :!class /Script/Engine.MaterialInstanceDynamic)
      (Effects|Components|ParticleSystem|SetFloatParameter _returnvalue_2 "LightningReactive" 0.0))))

(event EventOnDayPhaseChanged
  (params (CurrentDayPhase :type EFortDayPhase%20Enum :default "Morning") (PreviousDayPhase :type EFortDayPhase%20Enum :default "Morning") (bAtCreation :type Boolean)))

(event EventOnWeaponChanged
  (params (NewWeapon :type Fort%20Weapon%20Object%20Reference) (OldWeapon :type Fort%20Weapon%20Object%20Reference)))

(event EventOnWeaponFired)

(event EventOnNewWeapons
  (params (WeaponsAsItems :type Array%20of%20Fort%20Item%20Object%20References)))

(event EventOnCurieStateAttached
  (params (Tag :type Gameplay%20Tag%20Structure%20%28by%20ref%29)))

(event EventOnCurieStateDetached
  (params (Tag :type Gameplay%20Tag%20Structure%20%28by%20ref%29)))

(event EventOnBeginSkydiving
  (|CallParentFunction) ; This node is disabled and will not be called.\nDrag off pins to build functionality.
)

(event EventOnEndSkydiving
  (|CallParentFunction) ; This node is disabled and will not be called.\nDrag off pins to build functionality.
)

(event EventOnEnterVehicle
  (params (VehicleEntered :type Fort%20Vehicle%20Interface%20Interface%20%28by%20ref%29))
  (|CallParentFunction VehicleEntered) ; This node is disabled and will not be called.\nDrag off pins to build functionality.
)

(event EventOnExitVehicle
  (|CallParentFunction) ; This node is disabled and will not be called.\nDrag off pins to build functionality.
)

(event EventOnTargetingOrFirstPersonCameraChanged)

(event EventOnHitPawn
  (params (SelfActor :type Actor%20Object%20Reference) (OtherActor :type Actor%20Object%20Reference) (NormalImpulse :type Vector :default "0, 0, 0") (Hit :type Hit%20Result%20Structure%20%28by%20ref%29)))

(fn ConstructionScript
  (|CallParentFunction))
