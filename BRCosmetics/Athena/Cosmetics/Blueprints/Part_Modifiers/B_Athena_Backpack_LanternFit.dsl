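; Cosmetic part modifier for the LanternFit backpack that drives reactive material and idle particle effects.
; In gameplay, stat watchers for health gains, shield gains and kills pass an ID to GameplayReactivity, which writes ID and timestamp behind a DoOnce with a trailing 5-second delay.
; A randomly delayed event, sooner in frontend display, timestamps the yellow lights.
; Effects receive game-time timestamps as scalar and float parameters.
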(event EventOnPostResIn
  (bind _returnvalue (CustomCharacter|GetAssociatedPlayerPawn))
  (|CallParentFunction)
  (if (Reactive|Frontend|IsCosmeticObjectDisplayedinFrontendContext 0)
    (Utilities|FlowControl|Delay (Math|Random|RandomFloatinRange 2e+01 3e+01))
    (CallFunction|YellowLightsEvent)
    (else
      (Utilities|FlowControl|Sequence
        (:then_0
          (Cosmetics|WatchCosmeticStat _returnvalue "(TagName=\"StatManager.GainedHealthTimes\")" 0 false
            (:OnStatChanged
              (CallFunction|GameplayReactivity 1))))
        (:then_1
          (Cosmetics|WatchCosmeticStat _returnvalue "(TagName=\"StatManager.GainedShieldTimes\")" 0 false
            (:OnStatChanged
              (CallFunction|GameplayReactivity 2))))
        (:then_2
          (Cosmetics|WatchCosmeticStat _returnvalue "(TagName=\"StatManager.AthenaKills\")" 0 false
            (:OnStatChanged
              (CallFunction|GameplayReactivity 3))))
        (:then_3
          (Utilities|FlowControl|Delay (Math|Random|RandomFloatinRange 6e+01 1.2e+02))
          (CallFunction|YellowLightsEvent))))))

(event Custom|GameplayReactivity
  (params (ReactivityID :type Integer))
  (bind _returnvalue (CallFunction|GetIdleVFXComponent))
  (bind _returnvalue_2 (CustomCharacter|GetSkeletalMeshForPartOnPlayerPawn))
  (bind _returnvalue_3 (Utilities|Time|GetGameTimeinSeconds))
  (bind _returnvalue_4 (Math|Conversions|ToFloat(Integer) ReactivityID))
  (Utilities|FlowControl|DoOnce false
    (:Completed
      (Rendering|Material|SetScalarParameterValueonMaterials _returnvalue_2 "ReactivityID" _returnvalue_4)
      (Rendering|Material|SetScalarParameterValueonMaterials _returnvalue_2 "ReactivityTimestamp" _returnvalue_3)
      (Effects|Components|ParticleSystem|SetFloatParameter _returnvalue "ReactivityID" _returnvalue_4)
      (Effects|Components|ParticleSystem|SetFloatParameter _returnvalue "ReactivityTimestamp" _returnvalue_3)
      (Utilities|FlowControl|Delay 5.0))))

(event Custom|YellowLightsEvent
  (bind _returnvalue (Utilities|Time|GetGameTimeinSeconds))
  (Rendering|Material|SetScalarParameterValueonMaterials (CustomCharacter|GetSkeletalMeshForPartOnPlayerPawn) "YellowLightsTimestamp" _returnvalue)
  (Effects|Components|ParticleSystem|SetFloatParameter (CallFunction|GetIdleVFXComponent) "YellowLightsTimestamp" _returnvalue))

(fn ConstructionScript
  (|CallParentFunction))
