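; Part modifier for the PawJasmineSword back bling that plays a kill-reactive scrolling material effect on the backpack mesh.
; After spawn, the frontend branch previews the effect once following a delay.
; Otherwise, only emissive variant index 0 watches the Athena kills stat, replaying the effect on changes above zero.
; The Scrolling Effect timeline drives scroll step, intensity and swoosh material parameters, then switches reactivity off when finished.
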(event Custom|FE%20Reactivity
  (bind _returnvalue (CustomCharacter|GetAssociatedPlayerPawn))
  (bind _returnvalue_2 (Customization|GetSkeletalMeshforPartType _returnvalue "Backpack" :!class /Script/FortniteGame.FortPlayerPawn))
  (Variables|Default|SetScrolling%20Finished? false)
  (Rendering|Material|SetScalarParameterValueonMaterials _returnvalue_2 "Enable Reactivity?" 1.0)
  (bind (_direction _time_step _opacity) (Timeline|Scrolling%20Effect PlayFromStart)
    (:Update
      (Rendering|Material|SetScalarParameterValueonMaterials _returnvalue_2 "Scrolling - Time Step" _time_step)
      (Rendering|Material|SetScalarParameterValueonMaterials _returnvalue_2 "Scrolling - Intensity" _opacity)
      (Rendering|Material|SetScalarParameterValueonMaterials _returnvalue_2 "Swoosh Intensity" _opacity))
    (:Finished
      (Rendering|Material|SetScalarParameterValueonMaterials _returnvalue_2 "Enable Reactivity?" 0.0))))

(event EventOnPostResIn
  (bind _returnvalue (CustomCharacter|GetAssociatedPlayerPawn))
  (bind _returnvalue_2 (Customization|GetSkeletalMeshforPartType _returnvalue "Backpack" :!class /Script/FortniteGame.FortPlayerPawn))
  (|CallParentFunction)
  (if (Reactive|Frontend|ShouldDisplayFrontendReactivity 0 "Backpack")
    (Rendering|Material|SetScalarParameterValueonMaterials _returnvalue_2 "Enable Reactivity?" 0.0)
    (Utilities|FlowControl|Delay 6.0)
    (CallFunction|FEReactivity)
    (Utilities|FlowControl|Delay 1e+01)
    (else
      (Rendering|Material|SetScalarParameterValueonMaterials _returnvalue_2 "Enable Reactivity?" 0.0)
      (bind _index (CosmeticLoadout|GetVariantChannelActiveIndexforSlotActor "(TagName=\"LoadoutSlot.Backpack\")" _returnvalue "(TagName=\"Cosmetics.Variant.Channel.Emissive\")" (Utilities|Array|MakeArray "(TagName=\"Cosmetics.Variant.Property.Emissive0\")" "(TagName=\"Cosmetics.Variant.Property.Emissive1\")")))
      (if (not (Math|Conversions|ToBoolean(Integer) _index))
        (bind (_statname _statvalue) (Cosmetics|WatchCosmeticStat (CustomCharacter|GetAssociatedPlayerPawn) "(TagName=\"StatManager.AthenaKills\")" 0 false)
          (:OnStatChanged
            (if (> _statvalue 0)
              (Rendering|Material|SetScalarParameterValueonMaterials _returnvalue_2 "Enable Reactivity?" 1.0)
              (Timeline|Scrolling%20Effect PlayFromStart)
              (else
                (Rendering|Material|SetScalarParameterValueonMaterials _returnvalue_2 "Enable Reactivity?" 0.0)))))))))

(fn ConstructionScript
  (|CallParentFunction))
