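; Character part modifier for the Prize Tray backpack that drives its material reaction to kills, with a demo animation in the front end.
; After post-res-in it runs a toon front-end LUT replace, then branches on front end.
; In the front end, a delayed timeline drives the FEDemo scalar.
; In game, a watch on the Athena kills stat pushes game time, animation duration and kill count as stage to the materials.
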(event Custom|ReplayCountDown
  ;[0.141 0.4403 1 1 200] FRONTEND 🛒
  (Utilities|FlowControl|RetriggerableDelay 3.0)
  (bind (_direction _animationa) (Timeline|Timeline_0 PlayFromStart)
    (:Update
      (Rendering|Material|SetScalarParameterValueonMaterials (CustomCharacter|GetSkeletalMeshForPartOnPlayerPawn) "FEDemo" _animationa))))

(event EventOnPostResIn
  (bind _returnvalue (CustomCharacter|GetSkeletalMeshForPartOnPlayerPawn))
  ;[1 0.2176 0.2574 1 200] EVENT 🖥️
  (|CallParentFunction)
  (Utilities|ToonFrontendLUTReplace
    (:Out
      (if (CustomCosmeticModifier|IsinFrontEnd)
        ;[0.141 0.4403 1 1 200] FRONTEND 🛒
        (Utilities|FlowControl|RetriggerableDelay 3.0)
        (Timeline|Timeline_0 PlayFromStart)
        (else
          ;[1 0.7262 0.2176 1 200] ON KILL REACTIVITY ☠
          (bind (_statname _statvalue) (Cosmetics|WatchCosmeticStat (CustomCharacter|GetAssociatedPlayerPawn) "(TagName=\"StatManager.AthenaKills\")" 0 false)
            (:OnStatChanged
              (if (> _statvalue 0)
                (Rendering|Material|SetScalarParameterValueonMaterials _returnvalue "GameTimeFromBP" (Utilities|Time|GetGameTimeinSeconds))
                (Rendering|Material|SetScalarParameterValueonMaterials _returnvalue "Animation_Duration" 4.0)
                (Rendering|Material|SetScalarParameterValueonMaterials _returnvalue "CurrentStage" (Math|Conversions|ToFloat(Integer) _statvalue))))))))))

(fn ConstructionScript
  (|CallParentFunction))
