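; Cosmetic part modifier for the RowLiaison backpack that drives bool parameters on its idle VFX component from player actions.
; Firing a weapon pulses IsWeaponFired true for a short delay, then resets it.
; After res-in, a looping timer polls the pawn's crouch state and sets IsCrouched through DoOnce gates and delays, with a frontend-display check when not crouched.
; The other event overrides are disabled stubs.
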(event EventOnDayPhaseChanged
  (params (CurrentDayPhase :type EFortDayPhase%20Enum :default "Morning") (PreviousDayPhase :type EFortDayPhase%20Enum :default "Morning") (bAtCreation :type Boolean))
  (|CallParentFunction CurrentDayPhase PreviousDayPhase bAtCreation) ; This node is disabled and will not be called.\nDrag off pins to build functionality.
)

(event EventOnNewPlayerWeaponType
  (params (NewType :type EFortItemType%20Enum :default "WorldItem") (OldType :type EFortItemType%20Enum :default "WorldItem"))
  (|CallParentFunction NewType OldType) ; This node is disabled and will not be called.\nDrag off pins to build functionality.
)

(event EventOnWeaponChanged
  (params (NewWeapon :type Fort%20Weapon%20Object%20Reference) (OldWeapon :type Fort%20Weapon%20Object%20Reference))
  (|CallParentFunction NewWeapon OldWeapon) ; This node is disabled and will not be called.\nDrag off pins to build functionality.
)

(event EventOnNewWeapons
  (params (WeaponsAsItems :type Array%20of%20Fort%20Item%20Object%20References))
  (|CallParentFunction WeaponsAsItems) ; This node is disabled and will not be called.\nDrag off pins to build functionality.
)

(event EventOnCurieStateAttached
  (params (Tag :type Gameplay%20Tag%20Structure%20%28by%20ref%29))
  (|CallParentFunction Tag) ; This node is disabled and will not be called.\nDrag off pins to build functionality.
)

(event EventOnCurieStateDetached
  (params (Tag :type Gameplay%20Tag%20Structure%20%28by%20ref%29))
  (|CallParentFunction Tag) ; This node is disabled and will not be called.\nDrag off pins to build functionality.
)

(event EventOnBeginSkydiving
  (|CallParentFunction) ; This node is disabled and will not be called.\nDrag off pins to build functionality.
)

(event EventOnEndSkydiving
  (|CallParentFunction) ; This node is disabled and will not be called.\nDrag off pins to build functionality.
)

(event EventOnEnterVehicle
  (params (VehicleEntered :type Fort%20Vehicle%20Interface%20Interface%20%28by%20ref%29))
  (|CallParentFunction VehicleEntered) ; This node is disabled and will not be called.\nDrag off pins to build functionality.
)

(event EventOnExitVehicle
  (|CallParentFunction) ; This node is disabled and will not be called.\nDrag off pins to build functionality.
)

(event EventOnTargetingOrFirstPersonCameraChanged
  (|CallParentFunction) ; This node is disabled and will not be called.\nDrag off pins to build functionality.
)

(event EventOnHitPawn
  (params (SelfActor :type Actor%20Object%20Reference) (OtherActor :type Actor%20Object%20Reference) (NormalImpulse :type Vector :default "0, 0, 0") (Hit :type Hit%20Result%20Structure%20%28by%20ref%29))
  (|CallParentFunction SelfActor OtherActor NormalImpulse Hit) ; This node is disabled and will not be called.\nDrag off pins to build functionality.
)

(event EventOnWeaponFired
  (bind _returnvalue (CallFunction|GetIdleVFXComponent))
  (|CallParentFunction)
  (Effects|Components|ParticleSystem|SetBoolParameter _returnvalue "IsWeaponFired" true)
  (Utilities|FlowControl|Delay 0.1)
  (Effects|Components|ParticleSystem|SetBoolParameter _returnvalue "IsWeaponFired" false))

(event EventOnPostResIn
  (|CallParentFunction)
  (Utilities|Time|SetTimerbyEvent (delegate Custom|CheckCrouch) 0.1 true))

(event Custom|CheckCrouch
  (bind _returnvalue (CallFunction|GetIdleVFXComponent))
  (Utilities|FlowControl|Sequence
    (:then_0
      (if (AI|Components|NavMovement|IsCrouching (Variables|/Script/Engine.Character|GetCharacterMovement (CustomCharacter|GetAssociatedPlayerPawn)))
        (Utilities|FlowControl|DoOnce false
          (:Completed
            (Utilities|FlowControl|Sequence
              (:then_0
                (Effects|Components|ParticleSystem|SetBoolParameter _returnvalue "IsCrouched" true))
              (:then_1
                (Utilities|FlowControl|DoOnce false
                  (:Completed
                    (Utilities|FlowControl|Sequence
                      (:then_0
                        (Effects|Components|ParticleSystem|SetBoolParameter _returnvalue "IsCrouched" true))
                      (:then_1
                        _))))))
            (Utilities|FlowControl|Delay 0.1)
            (Effects|Components|ParticleSystem|SetBoolParameter _returnvalue "IsCrouched" false)))
        (elif (Reactive|Frontend|IsCosmeticObjectDisplayedinFrontendContext 0)
          (Utilities|FlowControl|Delay 0.3))))))

(fn ConstructionScript
  (|CallParentFunction))
