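; Backpack cosmetic part modifier for StunningMountain that drives its idle VFX from pickaxe, skydiving and emote state.
; PickaxeCheck negates a StunningMountain meta-tag test on the pickaxe via the part's anim instance, cached post-resource-init, returning false without one.
; The result sets bHasPickaxe on init, weapon change and skydive end, while skydiving forces it true with leftover debug prints.
; Emote start and end toggle Niagara bIsEmote.
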(event EventOnPostResIn
  (|CallParentFunction)
  (bind _ascustom_character_part_anim_instance (Utilities|Casting|CastToCustomCharacterPartAnimInstance (Components|SkeletalMesh|GetAnimInstance (Customization|GetSkeletalMeshforPartType (CustomCharacter|GetAssociatedPlayerPawn) (CustomCharacter|GetCharacterPartType) :!class /Script/FortniteGame.FortPlayerPawn))))
  (Variables|Default|SetAnim%20Instance _ascustom_character_part_anim_instance)
  (bind _returnvalue (CallFunction|PickaxeCheck))
  (Effects|Components|ParticleSystem|SetBoolParameter (CallFunction|GetIdleVFXComponent) "bHasPickaxe" _returnvalue))

(event EventOnNewPlayerWeaponType
  (params (NewType :type EFortItemType%20Enum :default "WorldItem") (OldType :type EFortItemType%20Enum :default "WorldItem"))
  (|CallParentFunction NewType OldType)
  (Utilities|IsSkydiving (CustomCharacter|GetAssociatedPlayerPawn)
    (:true
      (Effects|Components|ParticleSystem|SetBoolParameter (CallFunction|GetIdleVFXComponent) "bHasPickaxe" true)
      (Development|PrintString "fx should be full on" true true "(R=0.000000,G=1.000000,B=0.388174,A=1.000000)" 1e+01))
    (:false
      (bind _returnvalue (CallFunction|PickaxeCheck))
      (Effects|Components|ParticleSystem|SetBoolParameter (CallFunction|GetIdleVFXComponent) "bHasPickaxe" _returnvalue))))

(event EventOnBeginSkydiving
  (|CallParentFunction)
  (Effects|Components|ParticleSystem|SetBoolParameter (CallFunction|GetIdleVFXComponent) "bHasPickaxe" true)
  (Development|PrintString "fx should be full on" true true "(R=0.000000,G=1.000000,B=0.388174,A=1.000000)" 1e+01))

(event EventOnBeginEmote
  (params (MontageItemDef :type Fort%20Item%20Definition%20Object%20Reference) (PawnEmoting :type Fort%20Pawn%20Object%20Reference))
  (bind _asniagara_particle_system_component (Utilities|Casting|CastToNiagaraComponent (CallFunction|GetIdleVFXComponent)))
  (Niagara|SetNiagaraVariable(Bool) _asniagara_particle_system_component "bIsEmote" true))

(event EventOnEndEmote
  (params (MontageItemDef :type Fort%20Item%20Definition%20Object%20Reference) (PawnEmoting :type Fort%20Pawn%20Object%20Reference))
  (bind _asniagara_particle_system_component (Utilities|Casting|CastToNiagaraComponent (CallFunction|GetIdleVFXComponent)))
  (Niagara|SetNiagaraVariable(Bool) _asniagara_particle_system_component "bIsEmote" false))

(event EventOnEndSkydiving
  (|CallParentFunction)
  (bind _returnvalue (CallFunction|PickaxeCheck))
  (Effects|Components|ParticleSystem|SetBoolParameter (CallFunction|GetIdleVFXComponent) "bHasPickaxe" _returnvalue))

(fn PickaxeCheck
  (bind _anim_instance (Variables|Default|GetAnim%20Instance))
  (Utilities|IsValid _anim_instance
    (:Is%20Valid
      (return (not (CosmeticTags|PickaxeItemDefinitionHasAnyMetaTag _anim_instance (Utilities|Array|MakeArray "(TagName=\"Cosmetics.MetaTags.StunningMountain\")") true))))
    (:Is%20Not%20Valid
      (return false))))

(fn ConstructionScript
  (|CallParentFunction))
