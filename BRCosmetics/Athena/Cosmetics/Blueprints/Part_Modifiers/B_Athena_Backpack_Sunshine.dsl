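; Part modifier for the Sunshine backpack cosmetic that plays Niagara and material reactions when its player gets kills.
; After post-res-in, it attaches to the player's backpack mesh and passes the active Parts variant index to Niagara.
; A watched AthenaKills stat above zero activates the effect and sets the material KillTime to game time plus an offset.
; The frontend previews that reaction after a delay, and a ReplayCountDown event replays it after a retriggerable delay.
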(event EventOnPostResIn
  (|CallParentFunction)
  (Utilities|FlowControl|Sequence
    (:then_0
      ;[0 0 0] ATTACHMENT 🧲
      (Transformation|AttachActorToComponent (Customization|GetSkeletalMeshforPartType (CustomCharacter|GetAssociatedPlayerPawn) (CustomCharacter|GetCharacterPartType) :!class /Script/FortniteGame.FortPlayerPawn) "root" "SnapToTarget" "SnapToTarget" "SnapToTarget"))
    (:then_1
      ;[0 0 0] 🎨VARIANT
      (bind _index (Class|FortCosmeticFunctionLibrary|GetVariantChannelActiveIndexForSlot "(TagName=\"LoadoutSlot.Backpack\")" (CosmeticLoadout|GetCosmeticLoadout) "(TagName=\"Cosmetics.Variant.Channel.Parts\")" (Utilities|Array|MakeArray "(TagName=\"Cosmetics.Variant.Property.Stage1\")" "(TagName=\"Cosmetics.Variant.Property.Stage2\")")))
      (Niagara|SetNiagaraVariable(Float) (Variables|Default|GetNS_Backpack_Sunshine) "VariantIndex" (Math|Conversions|ToFloat(Integer) _index)))
    (:then_2
      ;[1 0.7262 0.2176 1 60] On Kill Reactivity ☠
      (bind (_statname _statvalue) (Cosmetics|WatchCosmeticStat (CustomCharacter|GetAssociatedPlayerPawn) "(TagName=\"StatManager.AthenaKills\")" 0 false)
        (:OnStatChanged
          (if (> _statvalue 0)
            (Components|Activation|Activate (Variables|Default|GetNS_Backpack_Sunshine) true)
            ;;[1 1 1] Material Reactivity
            (Rendering|Material|SetScalarParameterValueonMaterials (CustomCharacter|GetSkeletalMeshForPartOnPlayerPawn) "KillTime" (+ 4.2 (Utilities|Time|GetGameTimeinSeconds)))))))
    (:then_3
      ;[0.141 0.4403 1 1 60] Frontend 🛒
      (if (Reactive|Frontend|ShouldDisplayFrontendReactivity 0 "Backpack")
        (Utilities|FlowControl|Delay 1.0)
        (Components|Activation|Activate (Variables|Default|GetNS_Backpack_Sunshine) true)
        (Rendering|Material|SetScalarParameterValueonMaterials (CustomCharacter|GetSkeletalMeshForPartOnPlayerPawn) "KillTime" (+ 4.2 (Utilities|Time|GetGameTimeinSeconds))) ; MaterialUpdate
        (Utilities|FlowControl|RetriggerableDelay 5.0) ; Good Appproximate Time
      ))))

(event Custom|ReplayCountDown
  ;[0.141 0.4403 1 1 60] Frontend 🛒
  (Utilities|FlowControl|RetriggerableDelay 5.0) ; Good Appproximate Time
  (Components|Activation|Activate (Variables|Default|GetNS_Backpack_Sunshine) true)
  (Rendering|Material|SetScalarParameterValueonMaterials (CustomCharacter|GetSkeletalMeshForPartOnPlayerPawn) "KillTime" (+ 4.2 (Utilities|Time|GetGameTimeinSeconds))) ; MaterialUpdate
)

(fn ConstructionScript
  (|CallParentFunction))
