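; Backpack part modifier for the TechnoAwaken cosmetic that makes its idle VFX react to the player's weapon and emotes.
; Sets the "PickaxeReactive" float on post-res-in and when switching to a harvest weapon, clearing it for other weapon types.
; Switching to the harvest weapon also spawns a sound attached to the part's skeletal mesh.
; Emote begin and end toggle the Niagara "bIsEmote" bool on the same idle VFX component.
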(event EventOnPostResIn
  (|CallParentFunction)
  (Effects|Components|ParticleSystem|SetFloatParameter (CallFunction|GetIdleVFXComponent) "PickaxeReactive" 1.0))

(event EventOnNewPlayerWeaponType
  (params (NewType :type EFortItemType%20Enum :default "WorldItem") (OldType :type EFortItemType%20Enum :default "WorldItem"))
  (bind _returnvalue (CallFunction|GetIdleVFXComponent))
  (|CallParentFunction NewType OldType)
  (if (Utilities|Enum|Equal(Enum) NewType "WeaponHarvest")
    (Effects|Components|ParticleSystem|SetFloatParameter _returnvalue "PickaxeReactive" 1.0)
    (CallFunction|TriggerAudio)
    (else
      (Effects|Components|ParticleSystem|SetFloatParameter _returnvalue "PickaxeReactive" 0.0))))

(event EventOnBeginEmote
  (params (MontageItemDef :type Fort%20Item%20Definition%20Object%20Reference) (PawnEmoting :type Fort%20Pawn%20Object%20Reference))
  (bind _asniagara_particle_system_component (Utilities|Casting|CastToNiagaraComponent (CallFunction|GetIdleVFXComponent)))
  (Niagara|SetNiagaraVariable(Bool) _asniagara_particle_system_component "bIsEmote" true))

(event Custom|TriggerAudio
  ;[0 0 0] Setup Audio Component
  (Audio|SpawnSoundAttached (Variables|Default|GetSound) (CustomCharacter|GetSkeletalMeshForPartOnPlayerPawn)))

(event EventOnEndEmote
  (params (MontageItemDef :type Fort%20Item%20Definition%20Object%20Reference) (PawnEmoting :type Fort%20Pawn%20Object%20Reference))
  (bind _asniagara_particle_system_component (Utilities|Casting|CastToNiagaraComponent (CallFunction|GetIdleVFXComponent)))
  (Niagara|SetNiagaraVariable(Bool) _asniagara_particle_system_component "bIsEmote" false))

(fn ConstructionScript
  (|CallParentFunction))
