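; Part modifier for the TimeInterval back bling, which makes its idle VFX react to player eliminations.
; After post-res-in it runs a toon frontend LUT replace and watches the AthenaKills stat.
; Positive stat changes play a timeline driving the "KillReactivity" particle parameter, plus an attached sound outside frontend display.
; For frontend display, a separate delayed timeline previews the effect, with a ReplayCountDown event to replay it.
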(event Custom|ReplayCountDown
  ;[0.141 0.4403 1 1 60] Frontend 🛒
  (Utilities|FlowControl|RetriggerableDelay 3.0)
  (bind (_direction _kill_reactivity) (Timeline|Timeline PlayFromStart)
    (:Update
      (Effects|Components|ParticleSystem|SetFloatParameter (CallFunction|GetIdleVFXComponent) "KillReactivity" _kill_reactivity))))

(event EventOnPostResIn
  (|CallParentFunction)
  (Utilities|FlowControl|Sequence
    (:then_0
      (Utilities|ToonFrontendLUTReplace))
    (:then_1
      ;[0.141 0.4403 1 1 60] Frontend 🛒
      (if (Reactive|Frontend|ShouldDisplayFrontendReactivity 0 "Backpack")
        (Utilities|FlowControl|RetriggerableDelay 3.0)
        (Timeline|Timeline PlayFromStart)))
    (:then_2
      ;[1 0.7262 0.2176 1 60] On Kill Reactivity ☠
      (bind (_statname _statvalue) (Cosmetics|WatchCosmeticStat (CustomCharacter|GetAssociatedPlayerPawn) "(TagName=\"StatManager.AthenaKills\")" 0 false)
        (:OnStatChanged
          (if (> _statvalue 0)
            (Utilities|FlowControl|Sequence
              (:then_0
                (bind (_direction _kill_reactivity) (Timeline|Kill%20Reactivity PlayFromStart)
                  (:Update
                    (Effects|Components|ParticleSystem|SetFloatParameter (CallFunction|GetIdleVFXComponent) "KillReactivity" _kill_reactivity))))
              (:then_1
                (if (not (Reactive|Frontend|ShouldDisplayFrontendReactivity 0 "Backpack"))
                  (Audio|SpawnSoundAttached "/BRCosmetics/Sounds/Backblings/TimeInterval/Backpack_TimeInterval_Cue.Backpack_TimeInterval_Cue" (CustomCharacter|GetSkeletalMeshForPartOnPlayerPawn)))))))))))

(fn ConstructionScript
  (|CallParentFunction))
