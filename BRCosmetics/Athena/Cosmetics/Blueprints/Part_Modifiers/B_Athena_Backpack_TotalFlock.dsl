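; Part modifier for the Total Flock backpack that drives three-stage elimination-reactive visuals and audio when the first Emissive variant is active.
; Activated cosmetic stages update the idle Niagara effect and mesh material, spawn a burst, and play sounds for stages one and two.
; In the frontend, a timeline previews the stages without these sounds.
; Briefly pulsing bKillPrevStage on the idle effect clears lingering lightning between stages.
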(event EventOnPostResIn
  (bind _returnvalue (CustomCharacter|GetSkeletalMeshForPartOnPlayerPawn))
  ; FRONT END
  (|CallParentFunction)
  (bind _index (CosmeticLoadout|GetVariantChannelActiveIndexforSlotActor "(TagName=\"LoadoutSlot.Backpack\")" (CustomCharacter|GetAssociatedPlayerPawn) "(TagName=\"Cosmetics.Variant.Channel.Emissive\")" (Utilities|Array|MakeArray "(TagName=\"Cosmetics.Variant.Property.Emissive1\")" "(TagName=\"Cosmetics.Variant.Property.Emissive2\")")))
  (Variables|Default|SetIs%20Reactive (== _index 0))
  (if (Reactive|Frontend|ShouldDisplayFrontendReactivity 0 "Backpack")
    (if (Variables|Default|GetIs%20Reactive)
      (Timeline|Timeline Play
        (:Stage%200
          ; FX
          (CallFunction|StageZero)
          (Rendering|Material|SetScalarParameterValueonMaterials _returnvalue "Reactivity - Stages" 0.0))
        (:Stage%201
          (CallFunction|StageOne)
          (Rendering|Material|SetScalarParameterValueonMaterials _returnvalue "Reactivity - Stages" 1.0))
        (:Stage%202
          (CallFunction|StageTwo)
          (Rendering|Material|SetScalarParameterValueonMaterials _returnvalue "Reactivity - Stages" 2.0))))))

(event EventOnDayPhaseChanged
  (params (CurrentDayPhase :type EFortDayPhase%20Enum :default "Morning") (PreviousDayPhase :type EFortDayPhase%20Enum :default "Morning") (bAtCreation :type Boolean)))

(event EventOnNewPlayerWeaponType
  (params (NewType :type EFortItemType%20Enum :default "WorldItem") (OldType :type EFortItemType%20Enum :default "WorldItem"))
  (|CallParentFunction NewType OldType) ; This node is disabled and will not be called.\nDrag off pins to build functionality.
)

(event EventOnWeaponChanged
  (params (NewWeapon :type Fort%20Weapon%20Object%20Reference) (OldWeapon :type Fort%20Weapon%20Object%20Reference)))

(event EventOnWeaponFired)

(event EventOnNewWeapons
  (params (WeaponsAsItems :type Array%20of%20Fort%20Item%20Object%20References)))

(event Custom|StageZero
  (bind _returnvalue (CallFunction|GetIdleVFXComponent))
  ; FX FRONT END EVENTS
  ;; IDLE FX
  (Effects|Components|ParticleSystem|SetIntParameter (CallFunction|GetIdleVFXComponent) "bKillStage" 0)
  ;; BURST FX
  (bind _returnvalue_2 (PartModifier|SpawnSystem|SpawnSystemAttachedPartModifier "/BRCosmetics/Accessories/FORT_Backpacks/Backpack_M_MED_TotalFlock/FX/NS_Backpack_TotalFlock_Burst.NS_Backpack_TotalFlock_Burst" (CustomCharacter|GetSkeletalMeshForPartOnPlayerPawn) "root")
    (:Success
      (Effects|Components|ParticleSystem|SetIntParameter _returnvalue_2 "bKillStage" 0)
      (Effects|Components|ParticleSystem|SetBoolParameter _returnvalue_2 "bIsFrontEndPreview" true)
      ; KILL LINGERING LIGHTNING
      (Effects|Components|ParticleSystem|SetIntParameter _returnvalue "bKillPrevStage" 1)
      (Utilities|FlowControl|Delay 0.1)
      (Effects|Components|ParticleSystem|SetIntParameter _returnvalue "bKillPrevStage" 0))))

(event EventOnCurieStateAttached
  (params (Tag :type Gameplay%20Tag%20Structure%20%28by%20ref%29)))

(event EventOnCurieStateDetached
  (params (Tag :type Gameplay%20Tag%20Structure%20%28by%20ref%29)))

(event EventOnBeginSkydiving
  (|CallParentFunction) ; This node is disabled and will not be called.\nDrag off pins to build functionality.
)

(event Custom|StageOne
  (bind _returnvalue (CallFunction|GetIdleVFXComponent))
  ; FX FRONT END EVENTS
  ;; IDLE FX
  (Effects|Components|ParticleSystem|SetIntParameter (CallFunction|GetIdleVFXComponent) "bKillStage" 1)
  ;; BURST FX
  (bind _returnvalue_2 (PartModifier|SpawnSystem|SpawnSystemAttachedPartModifier "/BRCosmetics/Accessories/FORT_Backpacks/Backpack_M_MED_TotalFlock/FX/NS_Backpack_TotalFlock_Burst.NS_Backpack_TotalFlock_Burst" (CustomCharacter|GetSkeletalMeshForPartOnPlayerPawn) "root")
    (:Success
      (Effects|Components|ParticleSystem|SetIntParameter _returnvalue_2 "bKillStage" 1)
      (Effects|Components|ParticleSystem|SetBoolParameter _returnvalue_2 "bIsFrontEndPreview" true)
      ; KILL LINGERING LIGHTNING
      (Effects|Components|ParticleSystem|SetIntParameter _returnvalue "bKillPrevStage" 1)
      (Utilities|FlowControl|Delay 0.1)
      (Effects|Components|ParticleSystem|SetIntParameter _returnvalue "bKillPrevStage" 0))))

(event EventOnEndSkydiving
  (|CallParentFunction) ; This node is disabled and will not be called.\nDrag off pins to build functionality.
)

(event EventOnEnterVehicle
  (params (VehicleEntered :type Fort%20Vehicle%20Interface%20Interface%20%28by%20ref%29))
  (|CallParentFunction VehicleEntered) ; This node is disabled and will not be called.\nDrag off pins to build functionality.
)

(event EventOnExitVehicle
  (|CallParentFunction) ; This node is disabled and will not be called.\nDrag off pins to build functionality.
)

(event Custom|StageTwo
  (bind _returnvalue (CallFunction|GetIdleVFXComponent))
  ; FX FRONT END EVENTS
  ;; IDLE FX
  (Effects|Components|ParticleSystem|SetIntParameter (CallFunction|GetIdleVFXComponent) "bKillStage" 2)
  ;; BURST FX
  (bind _returnvalue_2 (PartModifier|SpawnSystem|SpawnSystemAttachedPartModifier "/BRCosmetics/Accessories/FORT_Backpacks/Backpack_M_MED_TotalFlock/FX/NS_Backpack_TotalFlock_Burst.NS_Backpack_TotalFlock_Burst" (CustomCharacter|GetSkeletalMeshForPartOnPlayerPawn) "root")
    (:Success
      (Effects|Components|ParticleSystem|SetIntParameter _returnvalue_2 "bKillStage" 2)
      (Effects|Components|ParticleSystem|SetBoolParameter _returnvalue_2 "bIsFrontEndPreview" true)
      ; KILL LINGERING LIGHTNING
      (Effects|Components|ParticleSystem|SetIntParameter _returnvalue "bKillPrevStage" 1)
      (Utilities|FlowControl|Delay 0.1)
      (Effects|Components|ParticleSystem|SetIntParameter _returnvalue "bKillPrevStage" 0))))

(event EventOnTargetingOrFirstPersonCameraChanged)

(event EventOnHitPawn
  (params (SelfActor :type Actor%20Object%20Reference) (OtherActor :type Actor%20Object%20Reference) (NormalImpulse :type Vector :default "0, 0, 0") (Hit :type Hit%20Result%20Structure%20%28by%20ref%29)))

(event EventOnCosmeticStateActivated
  (params (Stat%20Name :type Name :default "None") (Stat%20Value :type Integer) (STAGE%20Num%20Being%20Activated :type Integer))
  (bind _returnvalue (CustomCharacter|GetSkeletalMeshForPartOnPlayerPawn))
  (bind _returnvalue_2 (CallFunction|GetIdleVFXComponent))
  ;[0.15 0 0.05434 0.5] ELIM REACTIVITY
  (|CallParentFunction Stat%20Name Stat%20Value STAGE%20Num%20Being%20Activated)
  (if (Variables|Default|GetIs%20Reactive)
    (Utilities|FlowControl|Sequence
      (:then_0
        (Effects|Components|ParticleSystem|SetIntParameter _returnvalue_2 "bKillStage" STAGE%20Num%20Being%20Activated)
        (CallFunction|OnStatChanged Stat%20Name Stat%20Value)
        (bind _returnvalue_3 (PartModifier|SpawnSystem|SpawnSystemAttachedPartModifier "/BRCosmetics/Accessories/FORT_Backpacks/Backpack_M_MED_TotalFlock/FX/NS_Backpack_TotalFlock_Burst.NS_Backpack_TotalFlock_Burst" (CustomCharacter|GetSkeletalMeshForPartOnPlayerPawn) "root")
          (:Success
            (Effects|Components|ParticleSystem|SetIntParameter _returnvalue_3 "bKillStage" STAGE%20Num%20Being%20Activated)
            ;; KILL LINGERING LIGHTNING
            (Effects|Components|ParticleSystem|SetIntParameter _returnvalue_2 "bKillPrevStage" 1)
            (Utilities|FlowControl|Delay 0.1)
            (Effects|Components|ParticleSystem|SetIntParameter _returnvalue_2 "bKillPrevStage" 0))))
      (:then_1
        (switch int STAGE%20Num%20Being%20Activated
          (:0
            (Rendering|Material|SetScalarParameterValueonMaterials _returnvalue "Reactivity - Stages" 0.0))
          (:1
            (Rendering|Material|SetScalarParameterValueonMaterials _returnvalue "Reactivity - Stages" 1.0)
            (CallFunction|ReactiveAudio "(TagName=\"SoundLibrary.Cosmetic.Other.One\")"))
          (:2
            (Rendering|Material|SetScalarParameterValueonMaterials _returnvalue "Reactivity - Stages" 2.0)
            (CallFunction|ReactiveAudio "(TagName=\"SoundLibrary.Cosmetic.Other.Two\")")))))))

(fn ReactiveAudio
  (params (Event%20Name :type Sound%20Library%20Tag%20Structure))
  (Networking|CanExecuteCosmeticEvents
    (:True
      (SoundLibrary|PlaySoundSimple (WorldSubsystems|GetSoundLibrarySubsystem) self (Utilities|Struct|MakeSimpleSoundSettings Event%20Name true false true true 1.0 1.0 0.0 0 0 "None" (CustomCharacter|GetSkeletalMeshForPartOnPlayerPawn)) :!class /Script/SoundLibrary.SoundLibrarySubsystem))))

(fn ConstructionScript
  (|CallParentFunction))
