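; Back bling cosmetic modifier for the TruckScale backpack that plays a kill-reactive effect on its idle VFX component.
; After res-in, a front-end branch waits, then switches the effect on in its Reactive state for a few seconds as a preview.
; A watcher on the owning pawn's AthenaKills stat activates the same effect when the count is above zero, deactivating it after a retriggerable delay.
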(event EventOnPostResIn
  (bind _returnvalue (CallFunction|GetIdleVFXComponent))
  (bind _returnvalue_2 (CallFunction|GetIdleVFXComponent))
  ;[1 0.2176 0.2574 1 200] EVENT 🖥️
  (|CallParentFunction)
  (Utilities|FlowControl|Sequence
    (:then_0
      ;[0.141 0.4403 1 1 200] FRONTEND 🛒
      (if (CustomCosmeticModifier|IsinFrontEnd)
        (Utilities|FlowControl|Delay 3.0)
        (Components|Activation|SetActive _returnvalue true false)
        (Effects|Components|ParticleSystem|SetBoolParameter _returnvalue "Reactive" true)
        (Utilities|FlowControl|Delay 4.0)
        (Components|Activation|SetActive _returnvalue false false)
        (Utilities|FlowControl|RetriggerableDelay 4.5)))
    (:then_1
      ;[1 0.7262 0.2176 1 200] ON KILL REACTIVITY ☠
      (bind (_statname _statvalue) (Cosmetics|WatchCosmeticStat (CustomCharacter|GetAssociatedPlayerPawn) "(TagName=\"StatManager.AthenaKills\")" 0 false)
        (:OnStatChanged
          (if (> _statvalue 0)
            (Components|Activation|SetActive _returnvalue_2 true false)
            (Effects|Components|ParticleSystem|SetBoolParameter _returnvalue_2 "Reactive" true)
            (Utilities|FlowControl|RetriggerableDelay 2.0)
            (Components|Activation|SetActive _returnvalue_2 false false)
            (else
              (Effects|Components|ParticleSystem|SetBoolParameter _returnvalue_2 "Reactive" false))))))))

(fn ConstructionScript
  (|CallParentFunction))
