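; Part modifier Blueprint for the Veiled backpack cosmetic, adding front-end effects and a reaction to player kills.
; After res-in in the front end, it waits briefly and attaches a Niagara effect to the "pack" socket; ReplayCountDown repeats this through a retriggerable delay.
; It also watches the pawn's kill stat, and on positive values spawns a delayed burst effect and sound at the Backpack_BR socket.
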(event Custom|ReplayCountDown
  ;[0.141 0.4403 1 1 200] FRONTEND 🛒
  (Utilities|FlowControl|RetriggerableDelay 3.0)
  (PartModifier|SpawnSystem|SpawnSystemAttachedPartModifier (Variables|Default|GetNiagaraFXAsset) (CustomCharacter|GetSkeletalMeshForPartOnPlayerPawn) "pack" "0, 0, 0" "0, 0, 0" "SnapToTarget"
    (:Success)))

(event EventOnPostResIn
  (bind _returnvalue (CustomCharacter|GetSkeletalMeshForPartOnPlayerPawn))
  ;[1 0.2176 0.2574 1 200] EVENT 🖥️
  (|CallParentFunction)
  (Utilities|FlowControl|Sequence
    (:then_0
      ;[0.141 0.4403 1 1 200] FRONTEND 🛒
      (if (CustomCosmeticModifier|IsinFrontEnd)
        (Utilities|FlowControl|RetriggerableDelay 3.0)
        (PartModifier|SpawnSystem|SpawnSystemAttachedPartModifier (Variables|Default|GetNiagaraFXAsset) (CustomCharacter|GetSkeletalMeshForPartOnPlayerPawn) "pack" "0, 0, 0" "0, 0, 0" "SnapToTarget"
          (:Success))))
    (:then_1
      ;[1 0.7262 0.2176 1 200] ON KILL REACTIVITY ☠
      (bind (_statname _statvalue) (Cosmetics|WatchCosmeticStat (CustomCharacter|GetAssociatedPlayerPawn) "(TagName=\"StatManager.AthenaKills\")" 0 false)
        (:OnStatChanged
          (if (> _statvalue 0)
            (Utilities|FlowControl|Delay 0.5)
            (Niagara|SpawnSystemAttached "/BRCosmetics/Accessories/FORT_Backpacks/Backpack_F_MED_Veiled/FX/NS_Veiled_Reactive_Burst.NS_Veiled_Reactive_Burst" _returnvalue "Backpack_BR" "0, 0, 0" "0, 0, 0" "SnapToTarget" true)
            (Audio|SpawnSoundatLocation (Variables|Default|GetKillReaction_Veiled_SFX) (Transformation|GetSocketLocation _returnvalue "Backpack_BR"))))))))

(fn ConstructionScript
  (|CallParentFunction))
