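; Pickaxe cosmetic Blueprint for BlondeJaw that drives time-based reactive material and VFX effects from impacts.
; On impact FX, it triggers an update in frontend display and checks the surface against weak-point surfaces to drive reactivity.
; MaterialUpdate stamps game time onto the impact VFX and into a ReactivityStartTime slot on the weapon materials, idle VFX and swing VFX.
; A counter modulo the max stack count cycles those slots.
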(event EventTick
  (params (DeltaSeconds :type Float%20%28single-precision%29)))

(event FortWeapon|EventOnPlayImpactFX
  (params (HitResult :type Hit%20Result%20Structure%20%28by%20ref%29) (ImpactPhysicalSurface :type EPhysicalSurface%20Enum :default "SurfaceType_Default") (SpawnedPSC :type FXSystem%20Component%20Object%20Reference))
  (Variables|Default|SetImpact%20VFX SpawnedPSC)
  (|CallParentFunction HitResult ImpactPhysicalSurface SpawnedPSC)
  (Utilities|FlowControl|Sequence
    (:then_0
      (if (Reactive|Frontend|IsCosmeticObjectDisplayedinFrontendContext self)
        (CallFunction|MaterialUpdate)))
    (:then_1
      (bind _output_get (Variables|Default|SetCurrent%20Impact%20Physical%20Surface%200 ImpactPhysicalSurface))
      (Variables|Default|SetIs%20weak%20point%20hit (Utilities|Array|ContainsItem (Variables|Default|GetWeakPoint%20Physical%20Surface%20Array_0) _output_get))
      (if (Variables|Default|GetIs%20weak%20point%20hit)
        (Effects|Components|ParticleSystem|SetFloatParameter (Variables|Default|GetImpact%20VFX) "ReactivityTime" (Utilities|Time|GetGameTimeinSeconds))
        (CallFunction|OnWeakPointHit)))))

(event Custom|OnWeakPointHit
  (CallFunction|MaterialUpdate))

(event Custom|MaterialUpdate
  (bind _reactivitymaxhitstacks (Variables|Default|GetReactivityMaxHitStacks))
  (bind _reactivitycounter (Variables|Default|GetReactivityCounter))
  (bind _returnvalue (Math|Integer|%25%28Integer%29 _reactivitycounter _reactivitymaxhitstacks))
  (bind _returnvalue_2 (Utilities|String|ToString(Integer) _returnvalue))
  (bind _returnvalue_3 (Utilities|String|Append "ReactivityStartTime_" _returnvalue_2))
  (bind _returnvalue_4 (Utilities|String|StringToName _returnvalue_3))
  (Effects|Components|ParticleSystem|SetFloatParameter (Variables|Default|GetImpact%20VFX) "ReactivityTime" (Utilities|Time|GetGameTimeinSeconds))
  (Math|Integer|IncrementInt (Variables|Default|GetReactivityCounter)
    (:%20%20
      (Rendering|Material|SetScalarParameterValueonMaterials (Variables|Visual|GetWeaponMesh) _returnvalue_4 (Utilities|Time|GetGameTimeinSeconds))
      (Effects|Components|ParticleSystem|SetFloatParameter (CallFunction|GetIdleVFXComponent) _returnvalue_4 (Utilities|Time|GetGameTimeinSeconds))
      (Effects|Components|ParticleSystem|SetFloatParameter (CallFunction|GetSwingVFXComponent) _returnvalue_4 (Utilities|Time|GetGameTimeinSeconds)))))

(fn ConstructionScript
  (|CallParentFunction))
