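; Character part modifier for the HonorBraceJoust cosmetic that adds kill-reactive material and particle effects.
; After res-in, it proceeds only if the Parts (toon check) and Particle (reactive) variants are at their first listed option.
; In the front end, a timeline previews reactivity through a material parameter and idle VFX toggles.
; A kill-stat watcher timestamps the material animation and pulses the VFX, using retriggerable delays to switch it off.
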(event EventOnPostResIn
  (bind _returnvalue (CustomCharacter|GetSkeletalMeshForPartOnPlayerPawn))
  (bind _returnvalue_2 (CallFunction|GetIdleVFXComponent))
  (bind _returnvalue_3 (CallFunction|GetIdleVFXComponent))
  (bind _returnvalue_4 (CallFunction|GetIdleVFXComponent))
  ;[1 0.2176 0.2574 1 200] EVENT 🖥️
  (|CallParentFunction)
  ;; Checks if variant is toon.
  (bind _index (CosmeticLoadout|GetVariantChannelActiveIndexforSlotActor "(TagName=\"LoadoutSlot.Character\")" (CustomCharacter|GetAssociatedPlayerPawn) "(TagName=\"Cosmetics.Variant.Channel.Parts\")" (Utilities|Array|MakeArray "(TagName=\"Cosmetics.Variant.Property.Stage1\")" "(TagName=\"Cosmetics.Variant.Property.Stage2\")")))
  (if (== _index 0)
    ;; Checks if reactive is on.
    (bind _index_2 (CosmeticLoadout|GetVariantChannelActiveIndexforSlotActor "(TagName=\"LoadoutSlot.Character\")" (CustomCharacter|GetAssociatedPlayerPawn) "(TagName=\"Cosmetics.Variant.Channel.Particle\")" (Utilities|Array|MakeArray "(TagName=\"Cosmetics.Variant.Property.Particle1\")" "(TagName=\"Cosmetics.Variant.Property.Particle2\")")))
    (if (== _index_2 0)
      (Utilities|FlowControl|Sequence
        (:then_0
          ;[0.141 0.4403 1 1 200] FRONTEND 🛒
          (if (CustomCosmeticModifier|IsinFrontEnd)
            (Utilities|FlowControl|Sequence
              (:then_0
                (Components|Activation|SetActive _returnvalue_3 true false)
                (Effects|Components|ParticleSystem|SetBoolParameter _returnvalue_3 "Reactive" false))
              (:then_1
                (bind (_direction _reactivity_fe) (Timeline|Reactivity%20FE PlayFromStart)
                  (:Update
                    (Rendering|Material|SetScalarParameterValueonMaterials (CustomCharacter|GetSkeletalMeshForPartOnPlayerPawn) "ReactivityPreview" _reactivity_fe))
                  (:EnableFX
                    (Effects|Components|ParticleSystem|SetBoolParameter _returnvalue_2 "Reactive" true))
                  (:DisableFX
                    (Effects|Components|ParticleSystem|SetBoolParameter _returnvalue_2 "Reactive" false))) ; Previously 13sec, now 14.13 sec to match animation
              ))))
        (:then_1
          ;[1 0.7262 0.2176 1 200] ON KILL REACTIVITY ☠
          (bind (_statname _statvalue) (Cosmetics|WatchCosmeticStat (CustomCharacter|GetAssociatedPlayerPawn) "(TagName=\"StatManager.AthenaKills\")" 0 false)
            (:OnStatChanged
              (Utilities|FlowControl|Sequence
                (:then_0
                  (Rendering|Material|SetScalarParameterValueonMaterials _returnvalue "GameTimeFromBP" (Utilities|Time|GetGameTimeinSeconds))
                  (Rendering|Material|SetScalarParameterValueonMaterials _returnvalue "AnimDuration" 4.0))
                (:then_1
                  (if (> _statvalue 0)
                    (Components|Activation|SetActive _returnvalue_4 true false)
                    (Effects|Components|ParticleSystem|SetBoolParameter _returnvalue_4 "Reactive" true)
                    (Utilities|FlowControl|RetriggerableDelay 3.0)
                    (Effects|Components|ParticleSystem|SetBoolParameter _returnvalue_4 "Reactive" false)
                    (Utilities|FlowControl|RetriggerableDelay 1.0)
                    (Components|Activation|SetActive _returnvalue_4 false false)
                    (else
                      (Effects|Components|ParticleSystem|SetBoolParameter _returnvalue_4 "Reactive" false))))))))))))

(fn ConstructionScript
  (|CallParentFunction))
