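; Glider Blueprint for the BriskReignGlacier cosmetic that adjusts a trail effect to match the equipped particle variant.
; When the chute becomes visible, it calls the parent handler, then reads the pawn's active glider particle-channel variant from a two-tag list.
; For the second variant (index 1), it sets the Niagara float "Variant" to 1 on the third trail component.
; Nothing here resets that value for other variants.
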(event EventOnChuteVisibilityChanged
  (params (bParachuteVisibility :type Boolean))
  (|CallParentFunction bParachuteVisibility)
  (if bParachuteVisibility
    (bind _index (CosmeticLoadout|GetVariantChannelActiveIndexforSlotActor "(TagName=\"LoadoutSlot.Glider\")" (Variables|FortPlayerParachute|GetPlayerPawn) "(TagName=\"Cosmetics.Variant.Channel.Particle\")" (Utilities|Array|MakeArray "(TagName=\"Cosmetics.Variant.Property.Particle1\")" "(TagName=\"Cosmetics.Variant.Property.Particle2\")")))
    (if (== _index 1)
      (Niagara|SetNiagaraVariable(Float) (Utilities|Array|Get(acopy) (Variables|FortPlayerParachute|GetTrailComponents) 2) "Variant" 1.0))))

(fn ConstructionScript
  (|CallParentFunction))
