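; Ranked Tandem glider cosmetic Blueprint that picks the parachute mesh material for the rank setup event.
; EventRankMaterialSetup runs the parent logic, then queries the active index of the glider loadout slot's Particle variant channel against the Particle1 and Particle2 properties.
; An index above zero puts the Unreal rank material on material slot 0, otherwise it uses Bronze.
; BeginPlay, overlap and Tick events are empty.
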(event EventBeginPlay)

(event Collision|EventActorBeginOverlap
  (params (OtherActor :type Actor%20Object%20Reference)))

(event EventTick
  (params (DeltaSeconds :type Float%20%28single-precision%29)))

(event Default|EventRankMaterialSetup
  (params (Index :type Integer))
  (bind _parachutemesh (Variables|Parachute|GetParachuteMesh))
  (|CallParentFunction Index)
  (bind _index (CosmeticLoadout|GetVariantChannelActiveIndexforSlotActor "(TagName=\"LoadoutSlot.Glider\")" (Variables|FortPlayerParachute|GetPlayerPawn) "(TagName=\"Cosmetics.Variant.Channel.Particle\")" (Utilities|Array|MakeArray "(TagName=\"Cosmetics.Variant.Property.Particle1\")" "(TagName=\"Cosmetics.Variant.Property.Particle2\")")))
  (Math|Integer|CompareInt _index
    (:>
      (Rendering|Material|SetMaterial _parachutemesh 0 (Variables|Rank|RankMaterials|GetUnreal%20Material)))
    (:==
      _)
    (:<
      _))
  (Rendering|Material|SetMaterial _parachutemesh 0 (Variables|Rank|RankMaterials|GetBronze%20Material)))

(fn ConstructionScript
  (|CallParentFunction))
