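; Cosmetic character part modifier for the Hybrid Boot that animates glow parameters on the materials of the part's skeletal mesh.
; On BeginPlay, a timeline drives the GlowAnimation scalar on those materials.
; The ReplayCountDown custom event uses a three-second retriggerable delay, then plays a second timeline that drives GlowReactive; its caller is not shown.
; The post-res-in event and construction script only call their parents.
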(event Custom|ReplayCountDown
  ;[0.141 0.4403 1 1 200] FRONTEND 🛒
  (Utilities|FlowControl|RetriggerableDelay 3.0)
  (bind (_direction _animationa) (Timeline|Timeline_0 PlayFromStart)
    (:Update
      (Rendering|Material|SetScalarParameterValueonMaterials (CustomCharacter|GetSkeletalMeshForPartOnPlayerPawn) "GlowReactive" _animationa))))

(event EventOnPostResIn
  ;[1 0.2176 0.2574 1 200] EVENT 🖥️
  (|CallParentFunction))

(event EventBeginPlay
  ;[1 0.7262 0.2176 1 200] ON KILL REACTIVITY ☠
  (bind (_direction _animationa) (Timeline|Timeline PlayFromStart)
    (:Update
      (Rendering|Material|SetScalarParameterValueonMaterials (CustomCharacter|GetSkeletalMeshForPartOnPlayerPawn) "GlowAnimation" _animationa))))

(fn ConstructionScript
  (|CallParentFunction))
