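; Part modifier for the CanvasPrint back bling that makes it react to the player's eliminations.
; After post-res-in, it either enables a fake audio loop material parameter for frontend reactivity display, or watches the AthenaKills cosmetic stat.
; When that stat is at least one, it stamps game time on the mesh materials and passes the count to the idle VFX.
; Outside the frontend it also plays an attached elimination sound.
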(event EventOnPostResIn
  (|CallParentFunction)
  (if (Reactive|Frontend|ShouldDisplayFrontendReactivity 0 "Backpack")
    (Rendering|Material|SetScalarParameterValueonMaterials (CustomCharacter|GetSkeletalMeshForPartOnPlayerPawn) "EnableFrontendFakeAudioLoop" 1.0)
    (else
      (bind (_statname _statvalue) (Cosmetics|WatchCosmeticStat (CustomCharacter|GetAssociatedPlayerPawn) "(TagName=\"StatManager.AthenaKills\")")
        (:OnStatChanged
          (if (>= _statvalue 1)
            (CallFunction|Reactivity _statvalue)))))))

(event Custom|Reactivity
  (params (KillCount :type Integer))
  (Rendering|Material|SetScalarParameterValueonMaterials (CustomCharacter|GetSkeletalMeshForPartOnPlayerPawn) "BP_GetTime" (Utilities|Time|GetGameTimeinSeconds))
  (Effects|Components|ParticleSystem|SetIntParameter (CallFunction|GetIdleVFXComponent) "KillReactivityCount" KillCount)
  (if (not (CustomCosmeticModifier|IsinFrontEnd))
    (Audio|SpawnSoundAttached "/BRCosmetics/Sounds/Backblings/CanvasPrint/Backpack_CanvasPrint_Elimination_Cue.Backpack_CanvasPrint_Elimination_Cue" (CustomCharacter|GetSkeletalMeshForPartOnPlayerPawn))))

(fn ConstructionScript
  (|CallParentFunction))
