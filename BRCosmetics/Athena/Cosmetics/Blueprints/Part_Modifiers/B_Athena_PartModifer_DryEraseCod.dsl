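; Cosmetic part modifier for the DryEraseCod part, a back bling by its sound path, that reacts to eliminations and music.
; The kill reactivity event stamps game time on the part's materials, passes the kill count to the idle VFX and plays a sound outside the front end.
; After res-in, a looping timer polls music amplitude from an audio-analysis parameter collection, with DoOnce-gated timelines driving FadingIn and FadingOut material parameters.
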(event Custom|Reactivity
  (params (KillCount :type Integer))
  (Rendering|Material|SetScalarParameterValueonMaterials (CustomCharacter|GetSkeletalMeshForPartOnPlayerPawn) "BP_GetTime" (Utilities|Time|GetGameTimeinSeconds))
  (Effects|Components|ParticleSystem|SetIntParameter (CallFunction|GetIdleVFXComponent) "KillReactivityCount" KillCount)
  (if (not (CustomCosmeticModifier|IsinFrontEnd))
    (Audio|SpawnSoundAttached "/BRCosmetics/Sounds/Backblings/CanvasPrint/Backpack_CanvasPrint_Elimination_Cue.Backpack_CanvasPrint_Elimination_Cue" (CustomCharacter|GetSkeletalMeshForPartOnPlayerPawn))))

(event EventOnEndEmote
  (params (MontageItemDef :type Fort%20Item%20Definition%20Object%20Reference) (PawnEmoting :type Fort%20Pawn%20Object%20Reference)))

(event EventOnBeginEmote
  (params (MontageItemDef :type Fort%20Item%20Definition%20Object%20Reference) (PawnEmoting :type Fort%20Pawn%20Object%20Reference)))

(event EventOnPostResIn
  (|CallParentFunction)
  (Utilities|Time|SetTimerbyEvent (delegate Custom|PlayingMusic) 0.05 true false 0.5))

(event Custom|PlayingMusic
  (bind _returnvalue (Rendering|Material|GetScalarParameterValue "/Game/Athena/Events/AudioAnalysis/MPC_Audio_Reactive_Parameters.MPC_Audio_Reactive_Parameters" "Amplitude_Music" :!class /Script/Engine.KismetMaterialLibrary))
  (if (> _returnvalue 0)
    (Variables|Default|SetPlayed true)
    (elif (Variables|Default|GetPlayed)
      (Utilities|FlowControl|DoOnce false :!label _label
        (:Completed
          (bind (_direction _fading_out) (Timeline|Timeline_0 PlayFromStart)
            (:Update
              (Rendering|Material|SetScalarParameterValueonMaterials (CustomCharacter|GetSkeletalMeshForPartOnPlayerPawn) "FadingOut" _fading_out))
            (:Finished
              (Variables|Default|SetPlayed false)
              (Rendering|Material|SetScalarParameterValueonMaterials (CustomCharacter|GetSkeletalMeshForPartOnPlayerPawn) "FadingIn" 0.0)))))
      (else
        (goto _label Reset))))
  (Utilities|FlowControl|DoOnce false
    (:Completed
      (bind (_direction_2 _fading_in) (Timeline|Timeline PlayFromStart)
        (:Update
          (Rendering|Material|SetScalarParameterValueonMaterials (CustomCharacter|GetSkeletalMeshForPartOnPlayerPawn) "FadingIn" _fading_in))))))

(fn ConstructionScript
  (|CallParentFunction))
