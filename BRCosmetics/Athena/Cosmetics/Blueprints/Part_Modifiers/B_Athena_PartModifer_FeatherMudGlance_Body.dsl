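; Body part modifier for the FeatherMudGlance character cosmetic that drives reactive glow and idle VFX.
; At post-res-in it caches the player pawn outside frontend and enables reactivity from the Particle1 variant.
; Tick checks tactical sprinting or sliding, updating VFX and fading material "reactivity" through flag-guarded timelines.
; In frontend it disables tick and plays a showcase timeline; a separate event feeds footstep index to the VFX.
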(event EventOnPostResIn
  ;[0 0 0 0.5] Set Player Ref
  (|CallParentFunction)
  ;[0 0.7179 0.8906 0.5] Glow reaction is activated ?
  (bind (_index _bindexisvalid) (Class|FortCosmeticFunctionLibrary|GetVariantChannelActiveIndexForItem "/BRCosmetics/Athena/Items/Cosmetics/Characters/Character_FeatherMudGlance.Character_FeatherMudGlance" (CosmeticLoadout|GetCosmeticLoadout) "(TagName=\"Cosmetics.Variant.Channel.Particle\")" (Utilities|Array|MakeArray "(TagName=\"Cosmetics.Variant.Property.Particle1\")")))
  (if _bindexisvalid
    (Variables|Default|SetReactive? true)
    (else
      (Variables|Default|SetReactive? false)))
  (if (not (Reactive|Frontend|IsCosmeticObjectDisplayedinFrontendContext 0))
    (bind _asfort_player_pawn_athena (Utilities|Casting|CastToFortPlayerPawnAthena (CustomCharacter|GetAssociatedPlayerPawn)))
    (Variables|Default|SetPlayerRef _asfort_player_pawn_athena)))

(event EventTick
  (params (DeltaSeconds :type Float%20%28single-precision%29))
  ;[0 0 0 0.5] FX Reactivity
  (if (Reactive|Frontend|IsCosmeticObjectDisplayedinFrontendContext 0)
    (Actor|Tick|SetActorTickEnabled false)
    (bind (_direction _reactivity) (Timeline|Timeline_1 PlayFromStart)
      (:Update
        (Effects|Components|ParticleSystem|SetBoolParameter (CallFunction|GetIdleVFXComponent) "bIsFrontEnd" (>= _reactivity 0.5))
        (Rendering|Material|SetScalarParameterValueonMaterials (CustomCharacter|GetSkeletalMeshForPartOnPlayerPawn) "reactivity" _reactivity)))
    (else
      (Actor|Tick|SetActorTickEnabled true)
      (CallFunction|ReactivityUpdate))))

(event Custom|ReactivityUpdate
  (bind _playerref (Variables|Default|GetPlayerRef))
  (bind _bistacticalsprinting (Variables|/Script/FortniteGame.FortPlayerPawn|GetbIsTacticalSprinting _playerref))
  (bind _bissliding (Variables|/Script/FortniteGame.FortPawn|GetbIsSliding _playerref))
  (bind _returnvalue (or _bissliding _bistacticalsprinting))
  ;[0 0 0 0.5] FX Reactivity
  (if (Variables|Default|GetReactive?)
    (Effects|Components|ParticleSystem|SetBoolParameter (CallFunction|GetIdleVFXComponent) "bIsSprinting" _returnvalue)
    ;[0 0.2918 0.3613] material Glow
    (if (or (Variables|Default|Getplaying%20animation) (Variables|Default|Getplaying%20animation2))
      _
      (elif _returnvalue
        (if (not (Variables|Default|GetFade%20In))
          (Variables|Default|Setplaying%20animation true)
          (bind (_direction _fade_in) (Timeline|Timeline PlayFromStart)
            (:Update
              (Rendering|Material|SetScalarParameterValueonMaterials (CustomCharacter|GetSkeletalMeshForPartOnPlayerPawn) "reactivity" _fade_in))
            (:Finished
              (Variables|Default|SetFade%20In true)
              (Variables|Default|Setplaying%20animation false))))
        (elif (Variables|Default|GetFade%20In)
          (Variables|Default|Setplaying%20animation2 true)
          (bind (_direction_2 _fade_out) (Timeline|Timeline_0 PlayFromStart)
            (:Update
              (Rendering|Material|SetScalarParameterValueonMaterials (CustomCharacter|GetSkeletalMeshForPartOnPlayerPawn) "reactivity" _fade_out))
            (:Finished
              (Variables|Default|SetFade%20In false)
              (Variables|Default|Setplaying%20animation2 false))))))))

(event Custom|Read%20Foot%20Plants
  ;[0 0 0 0.5] Foot Index for FX
  (Effects|Components|ParticleSystem|SetBoolParameter (CallFunction|GetIdleVFXComponent) "bFootIndex" (Math|Conversions|ToBoolean(Integer) (AthenaPawn|GetCachedFootstepIndex (Variables|Default|GetPlayerRef)))))

(fn ConstructionScript
  (|CallParentFunction))
