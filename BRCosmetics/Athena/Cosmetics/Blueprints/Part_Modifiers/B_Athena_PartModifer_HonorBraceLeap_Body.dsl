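; Part modifier for the HonorBraceLeap body cosmetic that drives a "Reactivity" material scalar from player movement and jumps.
; In game, tick accumulates pawn speed, weighted below a jog threshold, with clamping and decay, then smooths it into reactivity.
; Jumping suspends speed updates while a timeline, seeked from current reactivity or restarted if mid-play, drives the scalar, resetting reactivity on finish.
; In frontend display, tick is disabled and a separate timeline drives the scalar.
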(event EventTick
  (params (DeltaSeconds :type Float%20%28single-precision%29))
  (bind _maxspeedaccum (Variables|Default|GetMaxSpeedAccum))
  ;[0 0 0] Speed reactivity
  (if (Reactive|Frontend|IsCosmeticObjectDisplayedinFrontendContext 0)
    ; FE LOOP
    (Actor|Tick|SetActorTickEnabled false)
    (bind (_direction _reactivity_fe) (Timeline|Reactivity%20FE PlayFromStart)
      (:Update
        (Rendering|Material|SetScalarParameterValueonMaterials (CustomCharacter|GetSkeletalMeshForPartOnPlayerPawn) "Reactivity" _reactivity_fe)))
    (else
      (Actor|Tick|SetActorTickEnabled true)
      (if (not (Variables|Default|GetIsJumping))
        (bind _output_get (Variables|Default|SetSpeedAccum (+ (Math|Float|Lerp (* (Variables|Default|GetJogSpeedMulti) (Math|Vector|VectorLength (Transformation|GetVelocity (CustomCharacter|GetAssociatedPlayerPawn)))) (Math|Vector|VectorLength (Transformation|GetVelocity (CustomCharacter|GetAssociatedPlayerPawn))) (Math|Conversions|ToFloat(Boolean) (> (Math|Vector|VectorLength (Transformation|GetVelocity (CustomCharacter|GetAssociatedPlayerPawn))) (Variables|Default|GetSpeedThreshold)))) (Variables|Default|GetSpeedAccum))))
        (bind _output_get_2 (Variables|Default|SetCurrReactivity (Math|Interpolation|FInterpTo (Variables|Default|GetCurrReactivity) (Math|Float|Lerp 0.0 1.0 (/ (Math|Interpolation|FInterpTo (Math|Float|Clamp(Float) _output_get 0.0 _maxspeedaccum) 0.0 DeltaSeconds 0.7) _maxspeedaccum)) DeltaSeconds 0.7)))
        (Rendering|Material|SetScalarParameterValueonMaterials (CustomCharacter|GetSkeletalMeshForPartOnPlayerPawn) "Reactivity" _output_get_2)
        (Variables|Default|SetSpeedAccum (Math|Interpolation|FInterpTo (Math|Float|Clamp(Float) _output_get 0.0 _maxspeedaccum) 0.0 DeltaSeconds 0.7))))))

(event EventOnPawnJumped
  ;[0 0 0] Jump reactivity
  (|CallParentFunction)
  (Variables|Default|SetIsJumping true)
  (if (Variables|Default|GetIsJumpFinished)
    (Utilities|FlowControl|Sequence
      (:then_0
        (bind (_direction _progress) (Timeline|JumpReactivity SetNewTime :NewTime (* (Variables|Default|GetCurrReactivity) 1.5))
          (:Update
            (Rendering|Material|SetScalarParameterValueonMaterials (CustomCharacter|GetSkeletalMeshForPartOnPlayerPawn) "Reactivity" _progress)
            (Variables|Default|SetCurrReactivity _progress)
            (Variables|Default|SetIsJumpFinished false))
          (:Finished
            (Variables|Default|SetIsJumping false)
            (Variables|Default|SetCurrReactivity 0.0)
            (Variables|Default|SetIsJumpFinished true))))
      (:then_1
        (Timeline|JumpReactivity Play)))
    (else
      (Timeline|JumpReactivity PlayFromStart))))

(fn ConstructionScript
  (|CallParentFunction))
