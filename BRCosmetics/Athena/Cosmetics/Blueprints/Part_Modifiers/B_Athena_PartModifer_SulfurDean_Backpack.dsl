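; Part modifier for the Sulfur Dean backpack cosmetic that makes its idle VFX react to eliminations.
; After resonance-in, frontend display fires Reactivity once after a short delay, while otherwise it watches the owning pawn's AthenaKills stat and fires on values of at least one.
; Reactivity ignores the passed kill count, increments its own Eliminations counter and writes it to the idle particle system's KillReactivityCount parameter.
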(event EventOnPostResIn
  (|CallParentFunction)
  (if (Reactive|Frontend|ShouldDisplayFrontendReactivity 0 "Backpack")
    (Utilities|FlowControl|Delay 4.0)
    (CallFunction|Reactivity 0)
    (Utilities|FlowControl|Delay 14.5)
    (else
      (bind (_statname _statvalue) (Cosmetics|WatchCosmeticStat (CustomCharacter|GetAssociatedPlayerPawn) "(TagName=\"StatManager.AthenaKills\")" 0 false)
        (:OnStatChanged
          (if (>= _statvalue 1)
            (CallFunction|Reactivity _statvalue)))))))

(event Custom|Reactivity
  (params (KillCount :type Integer))
  (bind _result (Math|Integer|IncrementInt (Variables|Default|GetEliminations))
    (:%20%20
      (Effects|Components|ParticleSystem|SetIntParameter (CallFunction|GetIdleVFXComponent) "KillReactivityCount" _result))))

(fn ConstructionScript
  (|CallParentFunction))
