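; Character part modifier for the Academia SL SpatialTravel cosmetic that feeds live bone positions to the part's materials.
; After resurrection-in, it sets BP_UseCCPM and reads the loadout's Material variant, disabling tick for Mat1 (index 0) and enabling it otherwise.
; Each tick writes the spine_05 location as chest position and the neck_02-to-head offset as a direction vector.
; GameTimeFromBPForGlow is an unused stub returning zero.
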(event EventOnPostResIn
  (|CallParentFunction)
  (Rendering|Material|SetScalarParameterValueonMaterials (CustomCharacter|GetSkeletalMeshForPartOnPlayerPawn) "BP_UseCCPM" 1.0)
  (bind _index (Class|FortCosmeticFunctionLibrary|GetVariantChannelActiveIndexForSlot "(TagName=\"LoadoutSlot.Character\")" (CosmeticLoadout|GetCosmeticLoadout) "(TagName=\"Cosmetics.Variant.Channel.Material\")" (Utilities|Array|MakeArray :%5B0%5D "(TagName=\"Cosmetics.Variant.Property.Mat1\")" :%5B1%5D "(TagName=\"Cosmetics.Variant.Property.Mat2\")" :%5B2%5D "(TagName=\"Cosmetics.Variant.Property.Mat3\")" :%5B3%5D "(TagName=\"Cosmetics.Variant.Property.Mat4\")")))
  (if (== _index 0)
    (Actor|Tick|SetActorTickEnabled false)
    (else
      (Actor|Tick|SetActorTickEnabled true))))

(event EventTick
  (params (DeltaSeconds :type Float%20%28single-precision%29))
  (bind _returnvalue (CustomCharacter|GetSkeletalMeshForPartOnPlayerPawn))
  (Rendering|Material|SetVectorParameterValueonMaterials _returnvalue "BP_OffsetVector_Chest" (- (Transformation|GetSocketLocation _returnvalue "head") (Transformation|GetSocketLocation _returnvalue "neck_02")))
  (Rendering|Material|SetVectorParameterValueonMaterials _returnvalue "BP_ChestPosition" (Transformation|GetSocketLocation _returnvalue "spine_05")))

(fn GameTimeFromBPForGlow
  (params (A :type Float%20%28double-precision%29) (B :type Float%20%28double-precision%29))
  (return 0.0))

(fn ConstructionScript)
