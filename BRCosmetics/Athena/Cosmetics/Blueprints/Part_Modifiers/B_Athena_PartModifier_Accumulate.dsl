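; Athena cosmetic part modifier that plays a reactive laugh and streak material effect on a character part's skeletal mesh.
; Any stat change raises AC_KillTrigger, runs a timeline driving AC_LaughStretch and AC_Streaks, then clears the trigger.
; After res-in, if frontend Backpack reactivity should display, a second timeline previews the effect, repeating through LoopEffect after a short delay; otherwise the trigger is zeroed.
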(event Custom|LoopEffect
  (bind _returnvalue (CustomCharacter|GetSkeletalMeshForPartOnPlayerPawn))
  ;[1 1 1] FrontEnd Reactivity Preview
  (Rendering|Material|SetScalarParameterValueonMaterials _returnvalue "AC_KillTrigger" 1.0)
  (bind (_direction _laughcurve _streakcurve) (Timeline|Timeline_1 PlayFromStart)
    (:Update
      (Rendering|Material|SetScalarParameterValueonMaterials _returnvalue "AC_LaughStretch" _laughcurve)
      (Rendering|Material|SetScalarParameterValueonMaterials _returnvalue "AC_Streaks" _streakcurve))
    (:Finished
      (Rendering|Material|SetScalarParameterValueonMaterials _returnvalue "AC_KillTrigger" 0.0)
      (Utilities|FlowControl|Delay 2.0)
      (CallFunction|LoopEffect))))

(event EventOnPostResIn
  (bind _returnvalue (CustomCharacter|GetSkeletalMeshForPartOnPlayerPawn))
  ;[1 1 1] FrontEnd Reactivity Preview
  (|CallParentFunction)
  (if (Reactive|Frontend|ShouldDisplayFrontendReactivity 0 "Backpack")
    (Rendering|Material|SetScalarParameterValueonMaterials _returnvalue "AC_KillTrigger" 1.0)
    (Timeline|Timeline_1 PlayFromStart)
    (else
      (Rendering|Material|SetScalarParameterValueonMaterials _returnvalue "AC_KillTrigger" 0.0))))

(event EventOnStatChanged
  (params (State%20Name :type Name :default "None") (Stat%20Value :type Integer))
  (bind _returnvalue (CustomCharacter|GetSkeletalMeshForPartOnPlayerPawn))
  (|CallParentFunction State%20Name Stat%20Value)
  ;[1 1 1] Set Mesh Materials
  (Rendering|Material|SetScalarParameterValueonMaterials _returnvalue "AC_KillTrigger" 1.0)
  (bind (_direction _laughcurve _streakcurve) (Timeline|Timeline_0 PlayFromStart)
    (:Update
      (Rendering|Material|SetScalarParameterValueonMaterials _returnvalue "AC_LaughStretch" _laughcurve)
      (Rendering|Material|SetScalarParameterValueonMaterials _returnvalue "AC_Streaks" _streakcurve))
    (:Finished
      (Rendering|Material|SetScalarParameterValueonMaterials _returnvalue "AC_KillTrigger" 0.0))))

(fn ConstructionScript)
