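; Part modifier for the AgileRug back bling that reacts to the player's kill stat, with a delayed preview in the frontend.
; A stat change above zero outside frontend preview, or the preview itself, calls TriggerReactivity.
; It writes the kill count's tens and ones, plus game time, into the mesh's custom primitive data.
; Behind a DoOnce, it spawns the kill Niagara system and passes the active particle variant as VariantID.
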(event EventOnPostResIn
  (|CallParentFunction)
  (if (Reactive|Frontend|ShouldDisplayFrontendReactivity 0 "Backpack")
    (Utilities|FlowControl|Delay 2.0)
    (CallFunction|TriggerReactivity)
    (Utilities|FlowControl|Delay 8.0)))

(event EventOnStatChanged
  (params (State%20Name :type Name :default "None") (Stat%20Value :type Integer))
  (|CallParentFunction State%20Name Stat%20Value)
  (bind _output_get (Variables|Default|SetAthenaKillStat Stat%20Value))
  (if (> _output_get 0)
    (if (not (Reactive|Frontend|ShouldDisplayFrontendReactivity 0 "Backpack"))
      (CallFunction|TriggerReactivity))))

(event Custom|TriggerReactivity
  (Rendering|Material|SetCustomPrimitiveDataVector3 (CustomCharacter|GetSkeletalMeshForPartOnPlayerPawn) 2 (Math|Vector|MakeVector (Math|Conversions|ToFloat(Integer) (/ (Variables|Default|GetAthenaKillStat) 10)) (Math|Conversions|ToFloat(Integer) (Math|Integer|%25%28Integer%29 (Variables|Default|GetAthenaKillStat) 10)) (Utilities|Time|GetGameTimeinSeconds)))
  (Utilities|FlowControl|DoOnce false
    (:Completed
      (bind _returnvalue (PartModifier|SpawnSystem|SpawnSystemAttachedPartModifier "/BRCosmetics/Accessories/FORT_Backpacks/Backpack_M_MED_AgileRug/FX/NS_Backpack_AgileRug_Kill.NS_Backpack_AgileRug_Kill" (CustomCharacter|GetSkeletalMeshForPartOnPlayerPawn) "FX_Idle")
        (:Success
          (bind _variantid (CallFunction|VariantCheck))
          (Niagara|SetNiagaraVariable(Float) _returnvalue "VariantID" _variantid)
          (Utilities|FlowControl|Delay 5.0))))))

(fn ConstructionScript
  (|CallParentFunction))

(fn VariantCheck
  (bind _index (Class|FortCosmeticFunctionLibrary|GetVariantChannelActiveIndexForItem "/BRCosmetics/Athena/Items/Cosmetics/Backpacks/Backpack_AgileRug.Backpack_AgileRug" (CosmeticLoadout|GetCosmeticLoadout) "(TagName=\"Cosmetics.Variant.Channel.Particle\")" (Utilities|Array|MakeArray "(TagName=\"Cosmetics.Variant.Property.Particle1\")" "(TagName=\"Cosmetics.Variant.Property.Particle2\")" "(TagName=\"Cosmetics.Variant.Property.Particle3\")" "(TagName=\"Cosmetics.Variant.Property.Particle4\")")))
  (return (Math|Conversions|ToFloat(Integer) _index)))
