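; Character part modifier for the AirKringle cosmetic that makes its idle particle effect react to jumping and landing.
; Pawn jump and land events write the current game time into the JumpGameTime and LandGameTime particle parameters, leaving the timed response to the effect itself.
; When displayed in the frontend, it simulates the jump reaction after a short delay post res-in, then repeats it on a looping timer.
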(event EventOnPostResIn
  (|CallParentFunction)
  (if (Reactive|Frontend|IsCosmeticObjectDisplayedinFrontendContext 0)
    (Utilities|FlowControl|Delay 4.0)
    (CallFunction|JumpReactivity)
    (Utilities|Time|SetTimerbyEvent (delegate Custom|Jump%20Reactivity) 8.0 true)))

(event Custom|Jump%20Reactivity
  (Effects|Components|ParticleSystem|SetFloatParameter (CallFunction|GetIdleVFXComponent) "JumpGameTime" (Utilities|Time|GetGameTimeinSeconds)))

(event EventOnPawnJumped
  (|CallParentFunction)
  (Effects|Components|ParticleSystem|SetFloatParameter (CallFunction|GetIdleVFXComponent) "JumpGameTime" (Utilities|Time|GetGameTimeinSeconds)))

(event EventOnPawnLanded
  (|CallParentFunction)
  (Effects|Components|ParticleSystem|SetFloatParameter (CallFunction|GetIdleVFXComponent) "LandGameTime" (Utilities|Time|GetGameTimeinSeconds)))

(fn ConstructionScript
  (|CallParentFunction))
