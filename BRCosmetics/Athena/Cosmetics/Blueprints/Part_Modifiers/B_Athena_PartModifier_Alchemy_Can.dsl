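; Part modifier for the Alchemy Can back bling cosmetic that plays a water splash effect when the wearer lands.
; On the pawn-landed event it calls the parent handler, then spawns the WaterSplash Niagara system.
; The effect attaches, snapped, to the root of the skeletal mesh for this part on the player pawn.
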(event EventOnPawnLanded
  (|CallParentFunction)
  (PartModifier|SpawnSystem|SpawnSystemAttachedPartModifier "/BRCosmetics/Accessories/FORT_Backpacks/Backpack_F_MED_Alchemy_Can/VFX/NS_AlchemyCan_WaterSplash.NS_AlchemyCan_WaterSplash" (CustomCharacter|GetSkeletalMeshForPartOnPlayerPawn) "root" "0, 0, 0" "0, 0, 0" "SnapToTarget"))

(fn ConstructionScript)
