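; Character part modifier for the AlienAgent cosmetic that adjusts shadow casting on its Face part.
; On the post res-in event it calls the parent, then looks up the owning pawn's skeletal mesh for the associated part type.
; For Face parts it turns off contact shadows, and in the front end also disables shadow casting entirely.
; Other part types and the construction script add nothing beyond parent behavior.
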(event EventOnPostResIn
  (bind _returnvalue (CustomCharacter|GetAssociatedPlayerPawn))
  (bind _returnvalue_2 (CustomCharacter|GetCharacterPartType))
  (bind _returnvalue_3 (Customization|GetSkeletalMeshforPartType _returnvalue _returnvalue_2 :!class /Script/FortniteGame.FortPlayerPawn))
  (|CallParentFunction)
  (switch Utilities|FlowControl|Switch|SwitchonEFortCustomPartType _returnvalue_2
    (:Face
      (Rendering|SetCastContactShadow _returnvalue_3 false)
      (if (CustomCosmeticModifier|IsinFrontEnd)
        (Rendering|SetCastShadow _returnvalue_3 false)))))

(fn ConstructionScript
  (|CallParentFunction))
