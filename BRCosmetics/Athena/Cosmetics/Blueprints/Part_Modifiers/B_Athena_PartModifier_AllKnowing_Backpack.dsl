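; Cosmetic part modifier Blueprint for the AllKnowing backpack that syncs its idle VFX with the pack's open animation event.
; After res-in, it finds the part's skeletal mesh on the owning player pawn, casts the anim instance to M_MED_AllKnowing_Pack_AnimBP and binds to its Open event.
; The handler passes the Opening flag to the "Opened" particle parameter on the idle VFX component, alongside a debug "Open" print.
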(event EventOnPostResIn
  (|CallParentFunction)
  (bind _asm_med_all_knowing_pack_anim_bp (Utilities|Casting|CastToM_MED_AllKnowing_Pack_AnimBP (Components|SkeletalMesh|GetAnimInstance (Customization|GetSkeletalMeshforPartType (CustomCharacter|GetAssociatedPlayerPawn) (CustomCharacter|GetCharacterPartType) :!class /Script/FortniteGame.FortPlayerPawn))))
  (Default|BindEventtoOpenEvent _asm_med_all_knowing_pack_anim_bp (delegate Custom|Open%20Event)))

(event Custom|Open%20Event
  (params (Opening :type Boolean))
  (Development|PrintString "Open" true true "(R=1.000000,G=0.000000,B=0.000000,A=1.000000)" 5.0)
  (Effects|Components|ParticleSystem|SetBoolParameter (CallFunction|GetIdleVFXComponent) "Opened" Opening))

(fn ConstructionScript
  (|CallParentFunction))
