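; Cosmetic part modifier for the Anarchy Acres Farmer character that swaps smoke VFX when the pawn jumps and lands.
; On jump it deactivates the idle VFX component and spawns a debounced jumping smoke emitter, attached to the part's skeletal mesh at FXSocket with custom depth matched to the pawn.
; On landing it reactivates the idle VFX.
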(event EventOnPawnJumped
  (Components|Activation|Deactivate (CallFunction|GetIdleVFXComponent))
  (PartModifier|SpawnSystem|SpawnEmitterAttachedPartModifier(Legacy) :EmitterTemplate (Variables|Default|GetJumping%20Local%20Smoke) :AttachToComponent (CustomCharacter|GetSkeletalMeshForPartOnPlayerPawn) :AttachPointName "FXSocket" :Scale "1.000000,1.000000,1.000000" :LocationType "SnapToTargetIncludingScale" :bAutoDestroy true :PoolingMethod "None" :bAutoActivate true :bMatchPawnCustomDepth true :bDebounce true :DebounceTime -1.0))

(event EventOnPawnLanded
  (Components|Activation|Activate (CallFunction|GetIdleVFXComponent) false))

(fn ConstructionScript)
