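; Cosmetic part modifier for the Angler character that makes its materials react when the player opens chests or gets eliminations.
; Outside the front end, post-res-in watches both stats, each restarting one shared timeline.
; Each timeline update calls WhenCosmeticPartApplied, storing the returned handle controller.
; Its callback writes the timeline track value into the "Reactivity" scalar on the part mesh's materials.
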(event EventOnPostResIn
  (bind _returnvalue (CustomCharacter|GetAssociatedPlayerPawn))
  (|CallParentFunction)
  (if (not (CustomCosmeticModifier|IsinFrontEnd))
    (Utilities|FlowControl|Sequence
      (:then_0
        (Cosmetics|WatchCosmeticStat _returnvalue "(TagName=\"StatManager.OpenChests\")"
          (:OnStatChanged
            (bind (_direction _newtrack_0) (Timeline|Timeline_0 PlayFromStart)
              (:Update
                (bind _delegatehandlecontroller_out (Cosmetics|WhenCosmeticPartApplied (CustomCharacter|GetAssociatedPlayerPawn) (delegate Custom|OnCosmeticPartApplied) true))
                (Variables|Default|SetDelegate%20Handle%20Controller _delegatehandlecontroller_out))))))
      (:then_1
        (Cosmetics|WatchCosmeticStat _returnvalue "(TagName=\"StatManager.AthenaKills\")"
          (:OnStatChanged
            (Timeline|Timeline_0 PlayFromStart)))))))

(event Custom|OnCosmeticPartApplied
  (params (Pawn :type Fort%20Player%20Pawn%20Object%20Reference) (Part :type Custom%20Character%20Part%20Object%20Reference) (MeshPart :type Skeletal%20Mesh%20Component%20Object%20Reference) (PartType :type EFortCustomPartType%20Enum :default "Head"))
  (Rendering|Material|SetScalarParameterValueonMaterials MeshPart "Reactivity" (stale EventOnPostResIn _newtrack_0)))

(fn ConstructionScript)
