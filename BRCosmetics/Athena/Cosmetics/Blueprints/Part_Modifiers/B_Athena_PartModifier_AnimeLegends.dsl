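; Cosmetic character part modifier for Anime Legends that adjusts the part's materials on the player pawn when the post-res-in event fires.
; After the parent logic, it runs ToonFrontendLUTReplace, then loops over the part's skeletal mesh materials.
; Any material instance parented to the Mockingbird body material has its Specular Strength overridden through a dynamic instance to reduce the gloss of the pants.
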(event EventOnPostResIn
  (bind _returnvalue (CustomCharacter|GetSkeletalMeshForPartOnPlayerPawn))
  (|CallParentFunction)
  (Utilities|ToonFrontendLUTReplace
    (:Out
      (for (_ _array_index) (Rendering|Material|GetMaterials _returnvalue)
        ;[1 1 1] Mockingbird reduce gloss of pants
        (if (== (Variables|/Script/Engine.MaterialInstance|GetParent (Class|BTechArtGlobalLibrary|GetDynamicMaterialInstance _returnvalue _array_index)) "/BRCosmetics/Characters/Player/Female/Large/Bodies/F_LRG_Mockingbird/Materials/M_F_Mockingbird_Body.M_F_Mockingbird_Body")
          (Rendering|Material|SetScalarParameterValue (Class|BTechArtGlobalLibrary|GetDynamicMaterialInstance _returnvalue _array_index) "Specular Strength" 0.55 :!class /Script/Engine.MaterialInstanceDynamic))))))

(fn ConstructionScript
  (|CallParentFunction))
