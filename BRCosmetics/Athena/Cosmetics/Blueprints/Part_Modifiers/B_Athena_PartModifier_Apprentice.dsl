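; Character part modifier for the Apprentice skin that adds kill and damage reactivity when its ReactiveOn meta tag is present.
; Variant tags pick the hair color, and watched DamageDealt and AthenaKills stats fire the reactions, with a kill-reaction preview in the frontend.
; Kills trigger VFX, Face hair overlay swaps with fresnel timelines, and Body eye fades.
; Damage accumulates into a capped, normalized material parameter, using retriggerable delays before resetting.
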(event EventOnPostResIn
  (bind _returnvalue (CustomCharacter|GetAssociatedPlayerPawn))
  (bind _returnvalue_2 (CustomCharacter|GetAssociatedPlayerPawn))
  (bind _returnvalue_3 (Cosmetics|GatherMetaTags _returnvalue_2 "/BRCosmetics/Athena/Items/Cosmetics/Characters/Character_Apprentice.Character_Apprentice" :!class /Script/FortniteGame.FortPlayerPawn))
  (bind _hairreactivecolor (Variables|Default|GetHairReactiveColor))
  (bind _dmi_hair (Variables|Default|GetDMI_Hair))
  (|CallParentFunction)
  ;[0 0.3147 1] Check Meta Tags of CID
  (bind _output_get (Variables|Default|SetReactivityEnabled (GameplayTags|HasTag _returnvalue_3 "(TagName=\"Cosmetics.MetaTags.ReactiveOn\")")))
  (if _output_get
    (Variables|Default|SetHairReactiveColor (select (GameplayTags|HasTag _returnvalue_3 "(TagName=\"Cosmetics.Variant.Property.Color.003\")") "(R=0.000000,G=1.000000,B=0.529096,A=1.000000)" (select (GameplayTags|HasTag _returnvalue_3 "(TagName=\"Cosmetics.Variant.Property.Color.002\")") "(R=1.000000,G=0.000000,B=0.951579,A=1.000000)" "(R=0.000000,G=0.685424,B=1.000000,A=1.000000)")))
    (Utilities|FlowControl|Sequence
      (:then_0
        ;[0.2818 1 0.7667] Init Triggers
        (bind (_statname _statvalue) (Cosmetics|WatchCosmeticStat _returnvalue "(TagName=\"StatManager.DamageDealt\")" 0 false)
          (:then
            (Cosmetics|WatchCosmeticStat _returnvalue "(TagName=\"StatManager.AthenaKills\")" 0 false
              (:OnStatChanged
                (CallFunction|OnKill))))
          (:OnStatChanged
            (Variables|Default|SetDamageDealt_All _statvalue)
            (CallFunction|OnDamageDealt))))
      (:then_1
        ;[1 0.3727 0.9421] Setup Outline Reactive Color
        (if (Utilities|Enum|Equal(Enum) (CustomCharacter|GetCharacterPartType) "Face")
          (bind _output_get_2 (Variables|Default|SetMI_Hair_Orig (Rendering|Material|GetOverlayMaterial (CustomCharacter|GetSkeletalMeshForPartOnPlayerPawn) false 0 :!class /Script/Engine.MeshComponent)))
          (bind _returnvalue_4 (Rendering|Material|CreateDynamicMaterialInstance _output_get_2 "None" "None" :!class /Script/Engine.KismetMaterialLibrary))
          (Variables|Default|SetDMI_Hair _returnvalue_4)
          ;;[1 0.6326 0.1244] Outline and InvertedHullOutline
          (Rendering|Material|SetVectorParameterValue _dmi_hair "OutlineColor" _hairreactivecolor :!class /Script/Engine.MaterialInstanceDynamic)
          (Rendering|Material|SetVectorParameterValue _dmi_hair "Color" _hairreactivecolor :!class /Script/Engine.MaterialInstanceDynamic)
          (Rendering|Material|SetScalarParameterValue _dmi_hair "Hide Element 01" 1.0 :!class /Script/Engine.MaterialInstanceDynamic)
          (Rendering|Material|SetScalarParameterValue _dmi_hair "Hide Element 02" 0.0 :!class /Script/Engine.MaterialInstanceDynamic)))
      (:then_2
        ;[0 0.4231 1] Reactivity Kill Anim in FE
        (if (Reactive|Frontend|ShouldDisplayFrontendReactivity 0 "Backpack")
          (Utilities|ToonFrontendLUTReplace
            (:Out
              (Utilities|FlowControl|Delay 3.0)
              (Variables|Default|SetKillReactivityInProgress true)
              (CallFunction|SpawnVFX)
              (CallFunction|ShowEyes)
              (Rendering|Material|SetScalarParameterValueonMaterials (CustomCharacter|GetSkeletalMeshForPartOnPlayerPawn) "Reactive_CT" (Utilities|Time|GetGameTimeinSeconds))
              (CallFunction|SwitchReactiveHairOn)
              (Utilities|FlowControl|RetriggerableDelay 6.0)
              (Variables|Default|SetKillReactivityInProgress false)
              (CallFunction|HideEyes)
              (CallFunction|SwitchReactiveHairOff)
              (Utilities|FlowControl|Delay 3.0)
              (if (Reactive|Frontend|ShouldDisplayFrontendReactivity 0 "Backpack")
                _))))))))

(event Custom|OnKill
  ;[1 0.2932 0] Reactivity Kill Anim
  (if (Variables|Default|GetReactivityEnabled)
    (Variables|Default|SetKillReactivityInProgress true)
    (CallFunction|SpawnVFX)
    (CallFunction|ShowEyes)
    (Rendering|Material|SetScalarParameterValueonMaterials (CustomCharacter|GetSkeletalMeshForPartOnPlayerPawn) "Reactive_CT" (Utilities|Time|GetGameTimeinSeconds))
    (CallFunction|SwitchReactiveHairOn)
    (Utilities|FlowControl|RetriggerableDelay 6.0)
    (Variables|Default|SetKillReactivityInProgress false)
    (CallFunction|HideEyes)
    (CallFunction|SwitchReactiveHairOff)
    (Utilities|FlowControl|Delay 2.0)))

(event Custom|OnDamageDealt
  (bind _returnvalue (CustomCharacter|GetSkeletalMeshForPartOnPlayerPawn))
  (bind _returnvalue_2 (CustomCharacter|GetSkeletalMeshForPartOnPlayerPawn))
  ;[1 0.1305 0.07081] Reactivity Damage Anim
  (if (Variables|Default|GetReactivityEnabled)
    (Variables|Default|SetDamageDealt_Current (- (Variables|Default|GetDamageDealt_All) (Variables|Default|GetDamageDealt_Before)))
    (Variables|Default|SetDamageCombined (Math|Integer|Min(Integer) (+ (Variables|Default|GetDamageDealt_Current) (Variables|Default|GetDamageCombined)) (Variables|Default|GetDamageCombined_Max)))
    (Variables|Default|SetDamageDealt_Before (Variables|Default|GetDamageDealt_All))
    (Variables|Default|SetDamageReactivityInProgress true)
    (CallFunction|ShowEyes)
    (Rendering|Material|SetScalarParameterValueonMaterials _returnvalue "Reactive_CurrentDamage" (* (/ 1.0 (Math|Conversions|ToFloat(Integer) (Variables|Default|GetDamageCombined_Max))) (Variables|Default|GetDamageCombined)))
    (Utilities|FlowControl|RetriggerableDelay 2.0)
    (Utilities|FlowControl|DoOnce false
      (:Completed
        (Rendering|Material|SetScalarParameterValueonMaterials _returnvalue "Reactive_Damage_CT" (Utilities|Time|GetGameTimeinSeconds))
        (Utilities|FlowControl|Delay 1.0)
        (Variables|Default|SetDamageReactivityInProgress false)
        (Variables|Default|SetDamageCombined 0)
        (Rendering|Material|SetScalarParameterValueonMaterials _returnvalue_2 "Reactive_CurrentDamage" 0.0)
        (Rendering|Material|SetScalarParameterValueonMaterials _returnvalue_2 "Reactive_Damage_CT" 99999.0)
        (CallFunction|HideEyes)))))

(event Custom|SwitchReactiveHairOn
  (bind _returnvalue (CustomCharacter|GetSkeletalMeshForPartOnPlayerPawn))
  ;[1 0.4023 0.2419] Hair Events
  (if (Utilities|Enum|Equal(Enum) (CustomCharacter|GetCharacterPartType) "Face")
    (Utilities|FlowControl|Sequence
      (:then_0
        (bind (_direction _blend _exponent) (Timeline|TimelineFresnelFadeIn PlayFromStart)
          (:Update
            (Rendering|Material|SetScalarParameterValueonMaterials _returnvalue "Fresnel_Blend" _blend)
            (Rendering|Material|SetScalarParameterValueonMaterials _returnvalue "Fresnel_Exponent" _exponent))))
      (:then_1
        (Utilities|FlowControl|Delay 0.5)
        (Rendering|Material|SetScalarParameterValueonMaterials _returnvalue "Hide Element 01" 1.0)
        (Rendering|Material|SetScalarParameterValueonMaterials _returnvalue "Hide Element 02" 0.0)
        (Rendering|Material|SetOverlayMaterial _returnvalue (Variables|Default|GetDMI_Hair) false 0 :!class /Script/Engine.MeshComponent)))))

(event Custom|ShowEyes
  (bind _returnvalue (CustomCharacter|GetSkeletalMeshForPartOnPlayerPawn))
  (bind _returnvalue_2 (CustomCharacter|GetSkeletalMeshForPartOnPlayerPawn))
  ;[1 0.4533 0.7355] Show & Hide Eyes
  (if (Utilities|Enum|Equal(Enum) (CustomCharacter|GetCharacterPartType) "Body")
    (if (not (Variables|Default|GetEyesAreVisiable))
      (Class|FortCosmeticMaterialBlueprintLibrary|SetScalarParameterValueonOverlayMaterial _returnvalue_2 "Hide Element 01")
      (Class|FortCosmeticMaterialBlueprintLibrary|SetScalarParameterValueonOverlayMaterial _returnvalue_2 "Hide Element 02" 1.0)
      (Rendering|Material|SetScalarParameterValueonMaterials _returnvalue "Hide Element 01" 0.0)
      (Variables|Default|SetEyesInTransition true)
      (Rendering|Material|SetScalarParameterValueonMaterials _returnvalue "EyeFadeIn_CT" (Utilities|Time|GetGameTimeinSeconds))
      (Utilities|FlowControl|Delay 0.351)
      (Variables|Default|SetEyesAreVisiable true)
      (Variables|Default|SetEyesInTransition false))))

(event Custom|HideEyes
  (bind _returnvalue (CustomCharacter|GetSkeletalMeshForPartOnPlayerPawn))
  ;[1 0.4533 0.7355] Show & Hide Eyes
  (if (Utilities|Enum|Equal(Enum) (CustomCharacter|GetCharacterPartType) "Body")
    (Utilities|FlowControl|RetriggerableDelay 1.0)
    (if (or (Variables|Default|GetKillReactivityInProgress) (Variables|Default|GetDamageReactivityInProgress))
      _
      (elif (Math|Boolean|NANDBoolean (Variables|Default|GetEyesAreVisiable) (Variables|Default|GetEyesInTransition))
        (Variables|Default|SetEyesInTransition true)
        (Rendering|Material|SetScalarParameterValueonMaterials _returnvalue "EyeFadeOut_CT" (Utilities|Time|GetGameTimeinSeconds))
        (Utilities|FlowControl|Delay 0.351)
        (Class|FortCosmeticMaterialBlueprintLibrary|SetScalarParameterValueonOverlayMaterial _returnvalue "Hide Element 01" 1.0)
        (Class|FortCosmeticMaterialBlueprintLibrary|SetScalarParameterValueonOverlayMaterial _returnvalue "Hide Element 02")
        (Rendering|Material|SetScalarParameterValueonMaterials _returnvalue "Hide Element 01" 1.0)
        (Rendering|Material|SetScalarParameterValueonMaterials _returnvalue "EyeFadeOut_CT" 99999.0)
        (Variables|Default|SetEyesAreVisiable false)
        (Variables|Default|SetEyesInTransition false)))))

(event Custom|SwitchReactiveHairOff
  (bind _returnvalue (CustomCharacter|GetSkeletalMeshForPartOnPlayerPawn))
  ;[1 0.4023 0.2419] Hair Events
  (if (Utilities|Enum|Equal(Enum) (CustomCharacter|GetCharacterPartType) "Face")
    (Utilities|FlowControl|Sequence
      (:then_0
        (bind (_direction _blendtimeline _fresneltimeline) (Timeline|TimelineFresnelFadeOut PlayFromStart)
          (:Update
            (Rendering|Material|SetScalarParameterValueonMaterials _returnvalue "Fresnel_Blend" _blendtimeline)
            (Rendering|Material|SetScalarParameterValueonMaterials _returnvalue "Fresnel_Exponent" _fresneltimeline))))
      (:then_1
        (Rendering|Material|SetScalarParameterValueonMaterials _returnvalue "Hide Element 01" 0.0)
        (Rendering|Material|SetScalarParameterValueonMaterials _returnvalue "Hide Element 02" 1.0)
        (Rendering|Material|SetOverlayMaterial _returnvalue (Variables|Default|GetMI_Hair_Orig) false 0 :!class /Script/Engine.MeshComponent)))))

(event Custom|SpawnVFX
  ;[1 0.7262 0.2176 1 40] ON KILL REACTIVITY ☠
  (Utilities|FlowControl|Delay 0.5)
  (Components|Activation|Activate (CallFunction|GetIdleVFXComponent) true)
  (if (not (Reactive|Frontend|ShouldDisplayFrontendReactivity 0 "Backpack"))
    (Audio|SpawnSoundAttached (Variables|Default|GetKillReaction_Apprentice_SFX) (CustomCharacter|GetSkeletalMeshForPartOnPlayerPawn))))

(fn ConstructionScript
  (|CallParentFunction))
