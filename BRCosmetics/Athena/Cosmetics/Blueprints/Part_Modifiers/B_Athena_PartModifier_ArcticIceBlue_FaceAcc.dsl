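; Character part modifier for the ArcticIceBlue face accessory that shows its scarf only while a supported backpack is visible.
; Pawn part-change, visibility and part-applied delegates track the backpack against a supported-parts list.
; Scarf changes update head and face material parameters, including vertex crunch, element hiding and overlay vertex-hide.
; Parameters are pushed immediately and on part application, with holo-mode fixes for ArcticIceTalus head wrap and antennas.
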(event EventOnPawnAndPartChanged
  (bind _returnvalue (CustomCharacter|GetAssociatedPlayerPawn))
  (|CallParentFunction)
  (Utilities|IsValid _returnvalue
    (:Is%20Valid
      (Utilities|FlowControl|Sequence
        (:then_0
          (EventDispatchers|BindEventtoOnPartChanged _returnvalue (EventDispatchers|CreateEvent :!event Handle%20Part%20Changed))
          (EventDispatchers|BindEventtoOnManagedPartVisibilityChanged _returnvalue (EventDispatchers|CreateEvent :!event Handle%20Part%20Visibility%20Changed)))
        (:then_1
          (Variables|Default|SetBackpack%20Visible (CharacterPart|IsPartVisibleAccordingtoCount _returnvalue "Backpack"))
          (bind _delegatehandlecontroller_out (Cosmetics|WhenCosmeticSpecificPartApplied _returnvalue (EventDispatchers|CreateEvent :!event Check%20Backpack%20Part) false "Backpack"))
          (Variables|Default|SetBackpack%20Delegate%20Controller _delegatehandlecontroller_out))
        (:then_2
          (CallFunction|CheckScarfVisible))))))

(event EventOnPostResIn
  (|CallParentFunction)
  (Default|UpdateMaterialParameters))

(event EventBeginPlay
  (|CallParentFunction))

(fn CheckBackpackPart
  (params (Pawn :type Fort%20Player%20Pawn%20Object%20Reference) (Part :type Custom%20Character%20Part%20Object%20Reference) (MeshPart :type Skeletal%20Mesh%20Component%20Object%20Reference) (PartType :type EFortCustomPartType%20Enum :default "Head"))
  (Utilities|IsValid Pawn
    (:Is%20Valid
      (bind _supported (CallFunction|IsBackpackSupported Part))
      (bind _output_get (LocalVariables|SetNew%20Supported%20Backpack _supported))
      (if (!= _output_get (Variables|Default|GetSupported%20Backpack))
        (Variables|Default|SetSupported%20Backpack (LocalVariables|GetNew%20Supported%20Backpack))
        (CallFunction|CheckScarfVisible)))))

(fn CheckScarfVisible
  (bind _output_get (LocalVariables|SetNew%20Scarf%20Visible (and (Variables|Default|GetSupported%20Backpack) (Variables|Default|GetBackpack%20Visible))))
  (if (!= _output_get (Variables|Default|GetScarf%20Visible))
    (Variables|Default|SetScarf%20Visible (LocalVariables|GetNew%20Scarf%20Visible))
    (Default|UpdateMaterialParameters)))

(fn HandlePartChanged
  (params (PartType :type EFortCustomPartType%20Enum :default "Head"))
  (bind _returnvalue (CustomCharacter|GetAssociatedPlayerPawn))
  (if (Utilities|Enum|Equal(Enum) PartType "Backpack")
    (Utilities|IsValid _returnvalue
      (:Is%20Valid
        (Cosmetics|WhenCosmeticSpecificPartApplied _returnvalue (EventDispatchers|CreateEvent :!event Check%20Backpack%20Part) false "Backpack")))))

(fn HandlePartVisibilityChanged
  (params (PartType :type EFortCustomPartType%20Enum :default "Head") (bVisible :type Boolean))
  (if (Utilities|Enum|Equal(Enum) PartType "Backpack")
    (if (!= (Variables|Default|GetBackpack%20Visible) (Variables|GetbVisible))
      (Variables|Default|SetBackpack%20Visible (Variables|GetbVisible))
      (CallFunction|CheckScarfVisible))))

(fn IsBackpackSupported
  (params (Character%20Part :type Custom%20Character%20Part%20Object%20Reference))
  (for _array_element (Variables|Default|GetSupported%20Character%20Parts)
    (if (== (Utilities|ResolveSoftReference _array_element) (Variables|GetCharacter%20Part))
      (return true)))
  (return false))

(fn PushFaceMaterialParameters
  (params (InputPin :type Skeletal%20Mesh%20Component%20Object%20Reference))
  (bind _scarf_visible (Variables|Default|GetScarf%20Visible))
  (bind _returnvalue (not _scarf_visible))
  (bind _returnvalue_2 (Math|Conversions|ToFloat(Boolean) _returnvalue))
  (bind _scarf_visible_2 (Variables|Default|GetScarf%20Visible))
  (Utilities|IsValid InputPin
    (:Is%20Valid
      ;[1 1 1] Hide Scarf in Holo Mode
      (bind _asmaterial_instance_dynamic (Utilities|Casting|CastToMaterialInstanceDynamic (Utilities|Array|Get(acopy) (Rendering|Material|GetMaterials InputPin) 1))
        (:then
          (Rendering|Material|SetScalarParameterValue _asmaterial_instance_dynamic "HT_CrunchVerts" _returnvalue_2 :!class /Script/Engine.MaterialInstanceDynamic))
        (:CastFailed
          _))
      ;[1 1 1] Hide ArcticIceTalus head wrap in FE holo mode
      (if (> (Utilities|String|FindSubstring (Utilities|GetDisplayName InputPin) "Talus") -1)
        (Rendering|Material|SetScalarParameterValue (Class|BTechArtGlobalLibrary|GetDynamicMaterialInstance InputPin) "HT_CrunchVerts" _returnvalue_2 :!class /Script/Engine.MaterialInstanceDynamic))
      (Rendering|Material|SetScalarParameterValueonMaterials InputPin "ArcticIce_Visibility" (Math|Conversions|ToFloat(Boolean) _scarf_visible))
      (Development|PrintString (Utilities|String|Append "Scarf Visible: " (Utilities|String|ToString(Boolean) _scarf_visible_2)))
      (bind _returnvalue_3 (Class|FortCosmeticMaterialBlueprintLibrary|GetorCreateDynamicOverlayMaterial InputPin))
      (Rendering|Material|SetScalarParameterValue _returnvalue_3 "VertexHide Threshold" (Math|Conversions|ToFloat(Boolean) (not _scarf_visible_2)) :!class /Script/Engine.MaterialInstanceDynamic))))

(fn PushHeadMaterialParameters
  (params (InputPin :type Skeletal%20Mesh%20Component%20Object%20Reference))
  (Utilities|IsValid InputPin
    (:Is%20Valid
      (Rendering|Material|SetScalarParameterValueonMaterials InputPin "Hide Element 10" (Math|Conversions|ToFloat(Boolean) (Variables|Default|GetScarf%20Visible))))))

(fn PushMaterialParameters
  (params (Pawn :type Fort%20Player%20Pawn%20Object%20Reference) (Part :type Custom%20Character%20Part%20Object%20Reference) (MeshPart :type Skeletal%20Mesh%20Component%20Object%20Reference) (PartType :type EFortCustomPartType%20Enum :default "Head"))
  (Utilities|IsValid MeshPart
    (:Is%20Valid
      (switch Utilities|FlowControl|Switch|SwitchonEFortCustomPartType PartType
        (:Head
          (CallFunction|PushHeadMaterialParameters (Variables|GetMeshPart)))
        (:Face
          (CallFunction|PushFaceMaterialParameters (Variables|GetMeshPart)))))))

(fn UpdateMaterialParameters
  (bind _returnvalue (CustomCharacter|GetAssociatedPlayerPawn))
  (Utilities|IsValid _returnvalue
    (:Is%20Valid
      ;[1 1 1] FORT-992376 push material parameters before binding
      (CallFunction|PushHeadMaterialParameters (Customization|GetSkeletalMeshforPartType _returnvalue "Head" :!class /Script/FortniteGame.FortPlayerPawn))
      (CallFunction|PushFaceMaterialParameters (Customization|GetSkeletalMeshforPartType _returnvalue "Face" :!class /Script/FortniteGame.FortPlayerPawn))
      (bind _delegatehandlecontroller_out (Cosmetics|WhenCosmeticPartApplied _returnvalue (EventDispatchers|CreateEvent :!event Push%20Material%20Parameters) true))
      (Variables|Default|SetMaterial%20Update%20Delegate%20Controller _delegatehandlecontroller_out)
      ;[1 1 1] FORT-553429: hide ArcticIceTalus antennas in FE holo mode
      (Rendering|Material|SetScalarParameterValue (Class|BTechArtGlobalLibrary|GetDynamicMaterialInstance (CustomCharacter|GetSkeletalMeshForPartOnPlayerPawn (FortPawn|GetModifierforPartType _returnvalue "Head"))) "HideElementsOnlyConsidersRedChannel" 0.0 :!class /Script/Engine.MaterialInstanceDynamic))))

(fn ConstructionScript
  (|CallParentFunction))
