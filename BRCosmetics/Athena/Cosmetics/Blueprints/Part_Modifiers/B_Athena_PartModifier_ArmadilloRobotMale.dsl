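; Cosmetic part modifier for the male Armadillo Robot character that drives a jet effect on the idle VFX component.
; Jumping, sliding and fast sprinting each add a named reason to a set, activating the jet; it deactivates when the last reason is removed.
; Sprint is checked every movement update by squared velocity and skipped mid-jump.
; Vehicle entry or exit and surface swimming force-clear all reasons, and swimming also hides the component.
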(event EventOnPostResIn
  (|CallParentFunction) ; This node is disabled and will not be called.\nDrag off pins to build functionality.
)

(event EventOnDayPhaseChanged
  (params (CurrentDayPhase :type EFortDayPhase%20Enum :default "Morning") (PreviousDayPhase :type EFortDayPhase%20Enum :default "Morning") (bAtCreation :type Boolean)))

(event EventOnNewPlayerWeaponType
  (params (NewType :type EFortItemType%20Enum :default "WorldItem") (OldType :type EFortItemType%20Enum :default "WorldItem"))
  (|CallParentFunction NewType OldType) ; This node is disabled and will not be called.\nDrag off pins to build functionality.
)

(event EventOnWeaponChanged
  (params (NewWeapon :type Fort%20Weapon%20Object%20Reference) (OldWeapon :type Fort%20Weapon%20Object%20Reference)))

(event EventOnWeaponFired)

(event EventOnNewWeapons
  (params (WeaponsAsItems :type Array%20of%20Fort%20Item%20Object%20References)))

(event EventOnCurieStateAttached
  (params (Tag :type Gameplay%20Tag%20Structure%20%28by%20ref%29)))

(event EventOnCurieStateDetached
  (params (Tag :type Gameplay%20Tag%20Structure%20%28by%20ref%29)))

(event EventOnBeginSkydiving
  (|CallParentFunction) ; This node is disabled and will not be called.\nDrag off pins to build functionality.
)

(event EventOnEndSkydiving
  (|CallParentFunction) ; This node is disabled and will not be called.\nDrag off pins to build functionality.
)

(event EventOnTargetingOrFirstPersonCameraChanged)

(event EventOnHitPawn
  (params (SelfActor :type Actor%20Object%20Reference) (OtherActor :type Actor%20Object%20Reference) (NormalImpulse :type Vector :default "0, 0, 0") (Hit :type Hit%20Result%20Structure%20%28by%20ref%29)))

(event EventOnPawnJumped
  (CallFunction|ActivateJetwithReason "jump"))

(event EventOnPawnLanded
  (CallFunction|DeactivateJetwithReason "jump"))

(event EventOnEnterVehicle
  (params (VehicleEntered :type Fort%20Vehicle%20Interface%20Interface%20%28by%20ref%29))
  (CallFunction|DeactivateJetwithReason "force" true))

(event EventOnExitVehicle
  (CallFunction|DeactivateJetwithReason "force" true))

(event EventOnBeginSliding
  (CallFunction|ActivateJetwithReason "slide"))

(event EventOnEndSliding
  (CallFunction|DeactivateJetwithReason "slide"))

(event EventOnCharacterMovementUpdated
  (params (DeltaSeconds :type Float%20%28single-precision%29) (OldLocation :type Vector :default "0, 0, 0") (OldVelocity :type Vector :default "0, 0, 0"))
  (bind _returnvalue (CustomCharacter|GetAssociatedPlayerPawn))
  (bind _returnvalue_2 (WaterBody|IsSurfaceSwimming _returnvalue))
  (|CallParentFunction DeltaSeconds OldLocation OldVelocity)
  (Utilities|FlowControl|Sequence
    (:then_1
      ;[1 1 1] Handle "sprinting" via velocity and forward movement
      (if (Math|Boolean|ANDBoolean (PawnStats|IsSprinting _returnvalue) (>= (Math|Vector|VectorLengthSquared (* (Transformation|GetVelocity _returnvalue) "1.000000,1.000000,1.000000")) (Variables|Default|GetSprintSpeedThreshhold)) (not (Utilities|Set|ContainsItem (Variables|Default|GetcurrentReasons) "jump")))
        (CallFunction|ActivateJetwithReason "sprint")
        (else
          (CallFunction|DeactivateJetwithReason "sprint"))))
    (:then_2
      ;[1 1 1] Hide VFX if swimming
      (if (Variables|Default|GetCurrentlySwimming)
        (if (not _returnvalue_2)
          (Development|SetHiddeninGame (CallFunction|GetIdleVFXComponent) false false)
          (Variables|Default|SetCurrentlySwimming false))
        (elif _returnvalue_2
          (Development|SetHiddeninGame (CallFunction|GetIdleVFXComponent) true false)
          (Variables|Default|SetCurrentlySwimming true)
          (CallFunction|DeactivateJetwithReason "swim" true))))))

(fn ActivateJetWithReason
  (params (Reason :type String))
  (if (not (Utilities|Set|ContainsItem (Variables|Default|GetcurrentReasons) Reason))
    (Utilities|Set|Add (Variables|Default|GetcurrentReasons) Reason)
    (Components|Activation|Activate (CallFunction|GetIdleVFXComponent) true)))

(fn DeactivateJetWithReason
  (params (Reason :type String) (Force :type Boolean))
  (bind _returnvalue (CallFunction|GetIdleVFXComponent))
  (bind _returnvalue_2 (CallFunction|GetIdleVFXComponent))
  (bind _currentreasons (Variables|Default|GetcurrentReasons))
  (if Force
    (Utilities|Set|Clear (Variables|Default|GetcurrentReasons))
    (if (Components|Activation|IsActive _returnvalue_2)
      (Components|Activation|Deactivate _returnvalue_2))
    (elif (Utilities|Set|ContainsItem _currentreasons Reason)
      (Utilities|Set|Remove _currentreasons Reason)
      (if (== (Utilities|Set|Length (Variables|Default|GetcurrentReasons)) 0)
        (if (Components|Activation|IsActive _returnvalue)
          (Components|Activation|Deactivate _returnvalue))))))

(fn DebugPrintReasons
  (bind _currentreasons (Variables|Default|GetcurrentReasons))
  (bind _returnvalue (Utilities|Set|Length _currentreasons))
  (LocalVariables|SetreasonString "REASONS: ")
  (if (== _returnvalue 0)
    (Development|PrintString "REASONS EMPTY" true true "(R=1.000000,G=0.739399,B=0.000000,A=1.000000)" 0.05)
    (else
      (bind _result (Utilities|Set|ToArray _currentreasons))
      (for _index (range _returnvalue)
        (LocalVariables|SetreasonString (Utilities|String|Append (LocalVariables|GetreasonString) "
" (Utilities|Array|Get(acopy) _result _index))))
      (Development|PrintString (LocalVariables|GetreasonString) true true "(R=1.000000,G=0.739399,B=0.000000,A=1.000000)" 0.05))))

(fn ConstructionScript
  (|CallParentFunction))
