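; Cosmetic part modifier for AshenMagnus that makes its part glow and drives backpack Niagara effects in reaction to the player's kills.
; On post-init, it caches the idle VFX component, then plays a frontend reactivity timeline or deactivates the effect and watches the kill stat.
; Timelines drive emissive brightness, plus Niagara reactivity and progressive variant for backpacks.
; Kills during playback's later portion rewind to a hold point; others reinitialize the effect.
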(event EventOnPostResIn
  (bind _returnvalue (CustomCharacter|GetCharacterPartType))
  (bind _returnvalue_2 (CustomCharacter|GetSkeletalMeshForPartOnPlayerPawn))
  (bind _niagara_system (Variables|Default|GetNiagara%20System))
  (bind _niagara_system_2 (Variables|Default|GetNiagara%20System))
  (bind _niagara_system_3 (Variables|Default|GetNiagara%20System))
  (|CallParentFunction)
  (bind _asniagara_particle_system_component (Utilities|Casting|CastToNiagaraComponent (CallFunction|GetIdleVFXComponent)))
  (Variables|Default|SetNiagara%20System _asniagara_particle_system_component)
  (if (Reactive|Frontend|ShouldDisplayFrontendReactivity 0 "Backpack")
    (bind (_direction _reactivity) (Timeline|TL_Reactivity_FE PlayFromStart)
      (:Update
        (Rendering|Material|SetScalarParameterValueonMaterials _returnvalue_2 "EmissiveBrightness" (Math|Float|Lerp 1.0 3.0 _reactivity))
        (switch Utilities|FlowControl|Switch|SwitchonEFortCustomPartType _returnvalue
          (:Backpack
            (bind _index (CosmeticLoadout|GetVariantChannelActiveIndexforSlotActor "(TagName=\"LoadoutSlot.Backpack\")" (CustomCharacter|GetAssociatedPlayerPawn) "(TagName=\"Cosmetics.Variant.Channel.Progressive\")" (Utilities|Array|MakeArray "(TagName=\"Cosmetics.Variant.Property.Stage1\")" "(TagName=\"Cosmetics.Variant.Property.Stage2\")")))
            (Niagara|SetNiagaraVariable(Float) _niagara_system_3 "Reactivity" _reactivity)
            (Niagara|SetNiagaraVariableByString(Int32) _niagara_system_3 "Variant" _index)))))
    (else
      (Components|Activation|Deactivate (Variables|Default|GetNiagara%20System))
      (Cosmetics|WatchCosmeticStat (CustomCharacter|GetAssociatedPlayerPawn) "(TagName=\"StatManager.AthenaKills\")" 0 false
        (:OnStatChanged
          (if (and (Components|Timeline|IsPlaying (Variables|B_Athena_PartModifier_AshenMagnus|GetTL_Reactivity)) (>= (Components|Timeline|GetPlaybackPosition (Variables|B_Athena_PartModifier_AshenMagnus|GetTL_Reactivity)) 0.75))
            (CallFunction|RetriggerReactivity)
            (else
              (Niagara|ReinitializeSystem _niagara_system_2)
              (bind (_direction_2 _reactivity_2) (Timeline|TL_Reactivity PlayFromStart)
                (:Update
                  (Rendering|Material|SetScalarParameterValueonMaterials _returnvalue_2 "EmissiveBrightness" (Math|Float|Lerp 1.0 3.0 _reactivity_2))
                  (switch Utilities|FlowControl|Switch|SwitchonEFortCustomPartType _returnvalue
                    (:Backpack
                      (bind _index_2 (CosmeticLoadout|GetVariantChannelActiveIndexforSlotActor "(TagName=\"LoadoutSlot.Backpack\")" (CustomCharacter|GetAssociatedPlayerPawn) "(TagName=\"Cosmetics.Variant.Channel.Progressive\")" (Utilities|Array|MakeArray "(TagName=\"Cosmetics.Variant.Property.Stage1\")" "(TagName=\"Cosmetics.Variant.Property.Stage2\")")))
                      (Niagara|SetNiagaraVariable(Float) _niagara_system "Reactivity" _reactivity_2)
                      (Niagara|SetNiagaraVariableByString(Int32) _niagara_system "Variant" _index_2))))
                (:Finished
                  (Components|Activation|Deactivate _niagara_system_2)
                  (Development|PrintString "ding"))))))))))

(event Custom|Retrigger%20Reactivity
  (Components|Timeline|SetPlaybackPosition (Variables|B_Athena_PartModifier_AshenMagnus|GetTL_Reactivity) 0.75 false true))

(fn ConstructionScript
  (|CallParentFunction))
