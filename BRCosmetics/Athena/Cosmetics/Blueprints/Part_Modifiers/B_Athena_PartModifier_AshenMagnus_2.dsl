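; Cosmetic part modifier for the Ashen Magnus character parts that makes them flare reactively when the player's kill stat changes.
; After the parent's PostResIn, it plays a timeline when frontend backpack reactivity should display, otherwise it watches the AthenaKills stat and replays one on each change.
; Timeline updates lerp the part mesh's EmissiveBrightness and, for backpack parts only, drive the idle Niagara effect's Reactivity variable.
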(event EventOnPostResIn
  (bind _returnvalue (CustomCharacter|GetSkeletalMeshForPartOnPlayerPawn))
  (bind _returnvalue_2 (CustomCharacter|GetCharacterPartType))
  (bind _returnvalue_3 (CallFunction|GetIdleVFXComponent))
  (|CallParentFunction)
  (if (Reactive|Frontend|ShouldDisplayFrontendReactivity 0 "Backpack")
    (bind (_direction _reactivity) (Timeline|Timeline_0 PlayFromStart)
      (:Update
        (Rendering|Material|SetScalarParameterValueonMaterials _returnvalue "EmissiveBrightness" (Math|Float|Lerp 1.0 3.0 _reactivity))
        (bind _asniagara_particle_system_component (Utilities|Casting|CastToNiagaraComponent _returnvalue_3))
        (switch Utilities|FlowControl|Switch|SwitchonEFortCustomPartType _returnvalue_2
          (:Backpack
            (Niagara|SetNiagaraVariable(Float) _asniagara_particle_system_component "Reactivity" _reactivity)))))
    (else
      (Cosmetics|WatchCosmeticStat (CustomCharacter|GetAssociatedPlayerPawn) "(TagName=\"StatManager.AthenaKills\")" 0 false
        (:OnStatChanged
          (bind (_direction_2 _reactivity_2) (Timeline|Timeline PlayFromStart)
            (:Update
              (Rendering|Material|SetScalarParameterValueonMaterials _returnvalue "EmissiveBrightness" (Math|Float|Lerp 1.0 3.0 _reactivity_2))
              (bind _asniagara_particle_system_component_2 (Utilities|Casting|CastToNiagaraComponent _returnvalue_3))
              (switch Utilities|FlowControl|Switch|SwitchonEFortCustomPartType _returnvalue_2
                (:Backpack
                  (Niagara|SetNiagaraVariable(Float) _asniagara_particle_system_component_2 "Reactivity" _reactivity_2))))))))))

(fn ConstructionScript
  (|CallParentFunction))
