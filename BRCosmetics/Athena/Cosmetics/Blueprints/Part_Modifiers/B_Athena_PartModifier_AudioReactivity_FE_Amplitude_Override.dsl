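; Cosmetic part modifier Blueprint that overrides audio-reactive amplitude response on a character part in the front end.
; On post res-in it runs the parent logic, then, only in the front end, sets the AmplitudeExponent material scalar on the part's skeletal mesh.
; A separate SetAndPlaySound function spawns the Apprentice back bling elimination-open cue attached to that mesh, with no caller in this graph.
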(event EventOnPostResIn
  (|CallParentFunction)
  (if (CustomCosmeticModifier|IsinFrontEnd)
    (Rendering|Material|SetScalarParameterValueonMaterials (CustomCharacter|GetSkeletalMeshForPartOnPlayerPawn) "AmplitudeExponent" 1.0)))

(fn SetAndPlaySound
  (Audio|SpawnSoundAttached "/BRCosmetics/Sounds/Backblings/Apprentice/BackBling_Apprentice_Elimination_Open_Cue.BackBling_Apprentice_Elimination_Open_Cue" (CustomCharacter|GetSkeletalMeshForPartOnPlayerPawn)))

(fn ConstructionScript
  (|CallParentFunction))
