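; Cosmetic part modifier Blueprint for the BabaYaga Athena character part, restarting its idle visual effect when the pawn lands.
; On the pawn-landed event it runs the parent handler, then fetches the idle VFX component and activates it with reset.
; The construction script is empty, and the file shows no other logic.
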(event EventOnPawnLanded
  (|CallParentFunction)
  (Components|Activation|Activate (CallFunction|GetIdleVFXComponent) true))

(fn ConstructionScript)
