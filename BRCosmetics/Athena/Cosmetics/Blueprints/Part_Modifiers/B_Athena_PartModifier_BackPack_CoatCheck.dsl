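; Backpack part modifier blueprint for the CoatCheck cosmetic back bling, adjusting how its mesh renders on the player pawn.
; After the post-res-in event runs the parent logic, it waits one tick.
; It then sets the custom depth stencil value of the part's skeletal mesh to 40, likely for stencil-driven post-process or material masking.
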(event EventOnPostResIn
  (|CallParentFunction)
  (Utilities|FlowControl|DelayUntilNextTick)
  (Rendering|SetCustomDepthStencilValue (CustomCharacter|GetSkeletalMeshForPartOnPlayerPawn) 40))

(fn ConstructionScript
  (|CallParentFunction))
