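; Part modifier for the MoleAcorn back bling that reacts to the player's eliminations.
; In game, it watches the AthenaKills stat and, on a higher count, sets KillCount on the idle Niagara effect and stamps BP_GetTime on the mesh materials to trigger WPO animation.
; An IsRespawned flag, re-armed after a delay, swallows some stat changes.
; In the frontend, a looping timer fakes eliminations with random counts.
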(event Custom|FE_Elim
  (bind _asniagara_particle_system_component (Utilities|Casting|CastToNiagaraComponent (CallFunction|GetIdleVFXComponent)))
  (Niagara|SetNiagaraVariable(Int32) _asniagara_particle_system_component "KillCount" (Math|Random|RandomInteger 999))
  (Rendering|Material|SetScalarParameterValueonMaterials (CustomCharacter|GetSkeletalMeshForPartOnPlayerPawn) "BP_GetTime" (Utilities|Time|GetGameTimeinSeconds)))

(event EventOnPostResIn
  (bind _killcount_moleacorn (Variables|Default|GetKillCount_MoleAcorn))
  (|CallParentFunction)
  (if (Reactive|Frontend|ShouldDisplayFrontendReactivity 0 "Backpack")
    (Utilities|Time|SetTimerbyEvent (delegate Custom|FE_Elim) 9.0 true false -6.0)
    (else
      (bind (_statname _statvalue) (Cosmetics|WatchCosmeticStat (CustomCharacter|GetAssociatedPlayerPawn) "(TagName=\"StatManager.AthenaKills\")")
        (:OnStatChanged
          (if (Variables|Default|GetIsRespawned)
            (Variables|Default|SetIsRespawned false)
            (Utilities|FlowControl|Sequence
              (:then_0
                (if (> _statvalue _killcount_moleacorn)
                  (bind _asniagara_particle_system_component (Utilities|Casting|CastToNiagaraComponent (CallFunction|GetIdleVFXComponent)))
                  (Niagara|SetNiagaraVariable(Int32) _asniagara_particle_system_component "KillCount" _statvalue)))
              (:then_1
                (if (> _statvalue _killcount_moleacorn)
                  ; WPO Animation on the Material
                  (Rendering|Material|SetScalarParameterValueonMaterials (CustomCharacter|GetSkeletalMeshForPartOnPlayerPawn) "BP_GetTime" (Utilities|Time|GetGameTimeinSeconds))
                  (Variables|Default|SetKillCount_MoleAcorn _statvalue)))
              (:then_2
                (Utilities|FlowControl|Delay 4.0)
                (Variables|Default|SetIsRespawned true)))))))))

(fn ConstructionScript
  (|CallParentFunction))
