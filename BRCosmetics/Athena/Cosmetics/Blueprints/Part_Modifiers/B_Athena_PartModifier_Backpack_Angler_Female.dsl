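; Backpack part modifier Blueprint for the female Angler cosmetic that reacts to player stat changes with a VFX pulse.
; When a stat changes to a value above zero, a timeline replays from the start.
; Each timeline update writes its curve value into the StatLerpKill float parameter on the idle VFX particle component.
; The curve name suggests the effect is meant for eliminations, though the stat name is not checked.
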(event EventOnStatChanged
  (params (State%20Name :type Name :default "None") (Stat%20Value :type Integer))
  (|CallParentFunction State%20Name Stat%20Value)
  (if (> Stat%20Value 0)
    (bind (_direction _statchangeelimination) (Timeline|Timeline_0 PlayFromStart)
      (:Update
        (Effects|Components|ParticleSystem|SetFloatParameter (CallFunction|GetIdleVFXComponent) "StatLerpKill" _statchangeelimination)))))

(fn ConstructionScript
  (|CallParentFunction))
