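; Backpack part modifier for the Assemble L cosmetic back bling that animates it in response to a tracked stat change.
; When the stat-changed event fires, it calls the parent handler, then gets the skeletal mesh for this part type on the associated player pawn.
; It plays the BatteryChange animation once on that mesh, regardless of the stat name or value.
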(event EventOnStatChanged
  (params (State%20Name :type Name :default "None") (Stat%20Value :type Integer))
  (|CallParentFunction State%20Name Stat%20Value)
  (Components|Animation|PlayAnimation (Customization|GetSkeletalMeshforPartType (CustomCharacter|GetAssociatedPlayerPawn) (CustomCharacter|GetCharacterPartType) :!class /Script/FortniteGame.FortPlayerPawn) "/BRCosmetics/Accessories/FORT_Backpacks/Backpack_M_MED_Assemble_L/Animations/Assemble_L_BatteryChange.Assemble_L_BatteryChange" false))

(fn ConstructionScript
  (|CallParentFunction))
