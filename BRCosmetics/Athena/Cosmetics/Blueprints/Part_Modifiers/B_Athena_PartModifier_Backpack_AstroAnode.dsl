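; Cosmetic part modifier for the AstroAnode back bling, reacting to stat changes, treated as kills, with a glitching counter.
; The first positive stat in-game, or a delayed front-end preview, fires a one-shot burst Niagara system on the backpack.
; Kill changes play a timeline-driven material glitch and display a separate counter cycling zero to nine, not the stat value.
; Landing briefly pulses bIsLanded on the idle VFX.
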(event EventOnPostResIn
  (|CallParentFunction)
  (Variables|Default|SetMesh (CustomCharacter|GetSkeletalMeshForPartOnPlayerPawn))
  (Utilities|FlowControl|Delay 2.0)
  (if (CustomCosmeticModifier|IsinFrontEnd)
    (CallFunction|TriggerReactivity)
    (CallFunction|UpdateKillCounter)
    (Utilities|FlowControl|Delay 5.0)))

(event EventOnPawnLanded
  (bind _returnvalue (CallFunction|GetIdleVFXComponent))
  (|CallParentFunction)
  (Effects|Components|ParticleSystem|SetBoolParameter _returnvalue "bIsLanded" true)
  (Utilities|FlowControl|Delay 0.1)
  (Effects|Components|ParticleSystem|SetBoolParameter _returnvalue "bIsLanded" false))

(event EventOnStatChanged
  (params (State%20Name :type Name :default "None") (Stat%20Value :type Integer))
  (|CallParentFunction State%20Name Stat%20Value)
  (if (> Stat%20Value 0)
    (if (not (CustomCosmeticModifier|IsinFrontEnd))
      (CallFunction|TriggerReactivity)
      (if (!= Stat%20Value (Variables|Default|GetCurrentKills))
        (Variables|Default|SetCurrentKills Stat%20Value)
        (CallFunction|UpdateKillCounter)))))

(event Custom|TriggerReactivity
  (Utilities|FlowControl|DoOnce false
    (:Completed
      (PartModifier|SpawnSystem|SpawnSystemAttachedPartModifier "/BRCosmetics/Accessories/FORT_Backpacks/Backpack_AstroAnode/FX/NS_Backpack_AstroAnode_Burst.NS_Backpack_AstroAnode_Burst" (Variables|Default|GetMesh) "backpack_root"
        (:Success)))))

(event Custom|TriggerCounterGlitchEffect
  (bind (_direction _glitchintensitytrack) (Timeline|GlitchEffectTimeline PlayFromStart)
    (:Update
      (Rendering|Material|SetScalarParameterValueonMaterials (Variables|Default|GetMesh) "Counter_Glitch" _glitchintensitytrack))))

(fn UpdateKillCounter
  (CallFunction|TriggerCounterGlitchEffect)
  (Rendering|Material|SetScalarParameterValueonMaterials (Variables|Default|GetMesh) "Kill_Number" (Math|Conversions|ToFloat(Integer) (Variables|Default|GetKillCounter)))
  (if (>= (Variables|Default|GetKillCounter) 9)
    (Variables|Default|SetKillCounter 0)
    (else
      (Math|Integer|IncrementInt (Variables|Default|GetKillCounter)))))

(fn ConstructionScript
  (|CallParentFunction))
