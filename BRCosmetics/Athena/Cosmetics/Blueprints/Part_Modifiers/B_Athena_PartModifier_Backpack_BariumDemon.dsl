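; Backpack part modifier for the BariumDemon cape that shows a stashed sword hilt depending on whether the BariumDemon pickaxe check passes.
; Weapon changes toggle the hilt through an equip-change Niagara burst, reveal/hide timelines on the Equipped_Hotspot material scalar, and the idle VFX ShowHilt parameter.
; Post res-in commits the initial hilt state and adjusts translucency sorting for toon outlines and cape layering.
; Tick conditionally applies holo materials for front-end holograms.
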(event EventTick
  (params (DeltaSeconds :type Float%20%28single-precision%29))
  (bind _returnvalue (CustomCharacter|GetAssociatedPlayerPawn))
  (bind _returnvalue_2 (CustomCharacter|GetCharacterPartType))
  (bind _returnvalue_3 (Customization|GetSkeletalMeshforPartType _returnvalue _returnvalue_2 :!class /Script/FortniteGame.FortPlayerPawn))
  (if (CustomCosmeticModifier|IsinFrontEnd)
    (if (CallFunction|IsFrontEndHolo)
      (if (not (== (Rendering|Material|GetMaterial _returnvalue_3 0) "/BRCosmetics/Accessories/FORT_Capes/Cape_M_MED_BariumDemon/FX/MI_MED_BariumDemon_Cape_FX_Inst.MI_MED_BariumDemon_Cape_FX_Inst"))
        (for _index (range 2)
          (Rendering|Material|SetMaterial _returnvalue_3 _index "/BRCosmetics/Accessories/FORT_Capes/Cape_M_MED_BariumDemon/FX/MI_MED_BariumDemon_Cape_FX_HOLO.MI_MED_BariumDemon_Cape_FX_HOLO"))))))

(event EventOnWeaponChanged
  (params (NewWeapon :type Fort%20Weapon%20Object%20Reference) (OldWeapon :type Fort%20Weapon%20Object%20Reference))
  (bind _bisdemonpickaxe (CallFunction|IsBariumDemonPickaxe))
  (if _bisdemonpickaxe
    (CallFunction|ToggleHiltStashedState false)
    (Variables|Default|SetisStashed false)
    (elif (Variables|Default|GetisStashed)
      _
      (else
        (CallFunction|ToggleHiltStashedState true)
        (Variables|Default|SetisStashed true)))))

(event EventOnPostResIn
  (bind _returnvalue (CustomCharacter|GetSkeletalMeshForPartOnPlayerPawn))
  (|CallParentFunction)
  (Utilities|FlowControl|Sequence
    (:then_0
      ;[1 1 1] For toon outlines on character, push backpack slightly higher in priority.
      (Utilities|IsValid _returnvalue
        (:Is%20Valid
          (Rendering|SetTranslucentSortPriority _returnvalue 0)
          (Rendering|SetTranslucencySortDistanceOffset _returnvalue -15.0))))
    (:then_1
      ;[0.2389 0.4583 0.02862] Render above the cape
      (Rendering|SetTranslucencySortDistanceOffset (CallFunction|GetIdleVFXComponent) -17.0)
      (bind _bisdemonpickaxe (CallFunction|IsBariumDemonPickaxe))
      (CallFunction|CommitHiltState (not _bisdemonpickaxe)))))

(event Custom|ToggleHiltStashedState
  (params (bHiltIsStashed :type Boolean))
  (bind _returnvalue (CustomCharacter|GetSkeletalMeshForPartOnPlayerPawn))
  (Utilities|FlowControl|Sequence
    (:then_0
      (if bHiltIsStashed
        (Utilities|FlowControl|Sequence
          (:then_0
            (bind (_direction _revealparam) (Timeline|RevealTimeline PlayFromStart)
              (:Update
                (Rendering|Material|SetScalarParameterValueonMaterials _returnvalue "Equipped_Hotspot" _revealparam))))
          (:then_1
            _))
        (else
          (Utilities|FlowControl|Sequence
            (:then_0
              _)
            (:then_1
              (bind (_direction_2 _revealparam_2) (Timeline|HideTimeline PlayFromStart)
                (:Update
                  (Rendering|Material|SetScalarParameterValueonMaterials _returnvalue "Equipped_Hotspot" _revealparam_2)))))))
      (bind _returnvalue_2 (PartModifier|SpawnSystem|SpawnSystemAttachedPartModifier "/BRCosmetics/Accessories/FORT_Capes/Cape_M_MED_BariumDemon/FX/NS_BariumDemon_EquipChange.NS_BariumDemon_EquipChange" _returnvalue "FX_Sword" "0, 0, 0" "0, 0, 0" "SnapToTarget")
        (:Success
          (Rendering|SetTranslucentSortPriority _returnvalue_2 2))))
    (:then_1
      (CallFunction|CommitHiltState bHiltIsStashed))))

(event Custom|CommitHiltState
  (params (IsStashed :type Boolean))
  (Effects|Components|ParticleSystem|SetBoolParameter (CallFunction|GetIdleVFXComponent) "ShowHilt" IsStashed)
  (Rendering|Material|SetScalarParameterValueonMaterials (CustomCharacter|GetSkeletalMeshForPartOnPlayerPawn) "Equipped_Hotspot" (Math|Float|SelectFloat 1.0 0.0 IsStashed)))

(fn IsBariumDemonPickaxe
  (bind _returnvalue (CustomCharacter|GetAssociatedPlayerPawn))
  (bind _returnvalue_2 (CustomCharacter|GetCharacterPartType))
  (bind _returnvalue_3 (Customization|GetSkeletalMeshforPartType _returnvalue _returnvalue_2 :!class /Script/FortniteGame.FortPlayerPawn))
  (Utilities|IsValid _returnvalue_3
    (:Is%20Valid
      (bind _ascustom_character_part_anim_instance (Utilities|Casting|CastToCustomCharacterPartAnimInstance (Components|SkeletalMesh|GetAnimInstance _returnvalue_3))
        (:then
          (return (CosmeticTags|PickaxeItemDefinitionHasAnyMetaTag _ascustom_character_part_anim_instance (Utilities|Array|MakeArray "(TagName=\"Cosmetics.MetaTags.BariumDemon\")") true)))
        (:CastFailed
          (return false))))
    (:Is%20Not%20Valid
      (return false))))

(fn IsFrontEndHolo
  (return (Variables|/Script/FortniteGame.FortPlayerPawn|GetbIsInFrontEndHologram (CustomCharacter|GetAssociatedPlayerPawn))))

(fn ConstructionScript
  (|CallParentFunction))
