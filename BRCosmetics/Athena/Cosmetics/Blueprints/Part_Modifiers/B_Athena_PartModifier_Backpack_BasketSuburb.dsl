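; Athena part modifier for the BasketSuburb backpack that makes its materials and VFX react to damage, chest opening, knockdown and sprinting.
; After res-in, outside frontend and preview and only if visible, it watches damage and chest stats and binds DBNO changes.
; Damage (unless sprinting or downed) spawns attached Niagara and pauses Idle Cry until a refreshed timer resets it.
; Chests briefly shift Animation Phase; DBNO and velocity-based sprinting set idle VFX flags.
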(event Custom|ResetTears
  ; Receiving damage reactivity
  (Variables|Default|SetReceivingDamage false)
  (Rendering|Material|SetScalarParameterValueonMaterials (CustomCharacter|GetSkeletalMeshForPartOnPlayerPawn) "Idle Cry" 1.0))

(event EventOnPostResIn
  (bind _returnvalue (CustomCharacter|GetSkeletalMeshForPartOnPlayerPawn))
  (bind _returnvalue_2 (CustomCharacter|GetSkeletalMeshForPartOnPlayerPawn))
  (|CallParentFunction)
  (if (or (Reactive|Frontend|IsCosmeticObjectDisplayedinFrontendContext 0) (Reactive|Frontend|IsCosmeticObjectDisplayedasPreview 0))
    _
    (elif (Rendering|IsVisible (CustomCharacter|GetSkeletalMeshForPartOnPlayerPawn))
      (Utilities|FlowControl|Sequence
        (:then_0
          ; Receiving damage reactivity
          (bind (_statname _statvalue) (Cosmetics|WatchCosmeticStat (CustomCharacter|GetAssociatedPlayerPawn) "(TagName=\"StatManager.DamageReceived\")" 0 false)
            (:OnStatChanged
              (if (Variables|Default|GetIsSprinting)
                _
                (elif (stale Custom|OnDBNOStateChanged bInIsDBNO)
                  (Components|Activation|Deactivate (Variables|Default|GetAttachedNS))
                  (Utilities|FlowControl|RetriggerableDelay 3.1)
                  (Variables|Default|SetReceivingDamage false)
                  (Variables|Default|SetDamageReceived 0)
                  (elif (> _statvalue (Variables|Default|GetDamageReceived))
                    (if (Variables|Default|GetReceivingDamage)
                      (Utilities|Time|ClearandInvalidateTimerbyHandle (stale _returnvalue_3))
                      (Components|Activation|Deactivate (Variables|Default|GetAttachedNS))
                      (CallFunction|SpawnReactivityFX)
                      (else
                        (CallFunction|SpawnReactivityFX)
                        (Rendering|Material|SetScalarParameterValueonMaterials _returnvalue "Idle Cry" 0.0)
                        (Rendering|Material|SetScalarParameterValueonMaterials _returnvalue "Animation Phase" 0.0)
                        (Variables|Default|SetReceivingDamage true)))
                    (bind _returnvalue_3 (Utilities|Time|SetTimerbyEvent (delegate Custom|ResetTears) 3.1))))))))
        (:then_1
          ; DBNO Reactivity
          (EventDispatchers|BindEventtoOnDBNOStateChanged (CustomCharacter|GetAssociatedPlayerPawn) (delegate Custom|OnDBNOStateChanged)))
        (:then_2
          ; Opean Chest Reactivity
          (Cosmetics|WatchCosmeticStat (CustomCharacter|GetAssociatedPlayerPawn) "(TagName=\"StatManager.OpenChests\")" 0 false
            (:OnStatChanged
              (Rendering|Material|SetScalarParameterValueonMaterials _returnvalue_2 "Idle Cry" 0.0)
              (Rendering|Material|SetScalarParameterValueonMaterials _returnvalue_2 "Animation Phase" 0.5)
              (Utilities|FlowControl|Delay 2.0)
              (Rendering|Material|SetScalarParameterValueonMaterials _returnvalue_2 "Idle Cry" 1.0)
              (Rendering|Material|SetScalarParameterValueonMaterials _returnvalue_2 "Animation Phase" 0.0))))))))

(event Custom|OnDBNOStateChanged
  (params (FortPawn :type Fort%20Pawn%20Object%20Reference) (bInIsDBNO :type Boolean))
  ; DBNO Reactivity
  (if bInIsDBNO
    (Effects|Components|ParticleSystem|SetBoolParameter (CallFunction|GetIdleVFXComponent) "IsDBNO" true)
    (else
      (Effects|Components|ParticleSystem|SetBoolParameter (CallFunction|GetIdleVFXComponent) "IsDBNO" false))))

(event EventOnLinearVelocity
  (params (NormalizedOverallVelocity :type Float%20%28double-precision%29) (NormalizedWalkVelocity :type Float%20%28double-precision%29) (VelocityLength :type Float%20%28double-precision%29))
  ; Sprinting Reactivity
  (if (not (Variables|Default|GetReceivingDamage))
    (bind _output_get (Variables|Default|SetIsSprinting (>= NormalizedOverallVelocity 0.8)))
    (Effects|Components|ParticleSystem|SetBoolParameter (CallFunction|GetIdleVFXComponent) "IsSprinting" _output_get)))

(fn SpawnReactivityFX
  (bind _returnvalue (PartModifier|SpawnSystem|SpawnSystemAttachedPartModifier (Variables|Default|GetReactivityNS) (CustomCharacter|GetSkeletalMeshForPartOnPlayerPawn) "root")
    (:Success
      (Variables|Default|SetAttachedNS _returnvalue))
    (:Failure
      _))
  (return))

(fn ConstructionScript
  (|CallParentFunction))
