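; Part modifier for the Big Chuggus backpack cosmetic that plays a particle effect when the pawn lands.
; On landing it calls the parent event, then spawns the landing emitter through the legacy part-modifier spawn call.
; The emitter snaps to the FX_Base socket of the character part's skeletal mesh, auto-destroys, is debounced and matches the pawn's custom depth.
; The construction script is empty.
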(event EventOnPawnLanded
  (|CallParentFunction)
  (PartModifier|SpawnSystem|SpawnEmitterAttachedPartModifier(Legacy) :EmitterTemplate "/BRCosmetics/Accessories/FORT_Backpacks/Effects/P_BigChuggus_Backpack_Landing.P_BigChuggus_Backpack_Landing" :AttachToComponent (CustomCharacter|GetSkeletalMeshForPartOnPlayerPawn) :AttachPointName "FX_Base" :Scale "1.000000,1.000000,1.000000" :LocationType "SnapToTargetIncludingScale" :bAutoDestroy true :PoolingMethod "None" :bAutoActivate true :bMatchPawnCustomDepth true :bDebounce true :DebounceTime -1.0))

(fn ConstructionScript)
