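; Character part modifier for the BikeMold backpack cosmetic that syncs its idle VFX with wing folding and jumping.
; On post-res-in, it gets the part's skeletal mesh anim instance, casts it to M_MED_BikeMold_Pack_AnimBP and binds to its FoldingWings dispatcher.
; That callback forwards the folding state to the idle particle component as a bool parameter.
; Pawn jumps pulse a separate jumped parameter on for half a second; other event overrides are empty or disabled.
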(event EventOnPostResIn
  (|CallParentFunction)
  (bind _asm_med_bike_mold_pack_anim_bp (Utilities|Casting|CastToM_MED_BikeMold_Pack_AnimBP (Components|SkeletalMesh|GetAnimInstance (Customization|GetSkeletalMeshforPartType (CustomCharacter|GetAssociatedPlayerPawn) (CustomCharacter|GetCharacterPartType) :!class /Script/FortniteGame.FortPlayerPawn))))
  (Default|BindEventtoFoldingWings _asm_med_bike_mold_pack_anim_bp (EventDispatchers|CreateEvent :!event Custom|FoldingWIngs)))

(event EventOnDayPhaseChanged
  (params (CurrentDayPhase :type EFortDayPhase%20Enum :default "Morning") (PreviousDayPhase :type EFortDayPhase%20Enum :default "Morning") (bAtCreation :type Boolean)))

(event Custom|FoldingWIngs
  (params (bIsFoldingWings :type Boolean))
  (Effects|Components|ParticleSystem|SetBoolParameter (CallFunction|GetIdleVFXComponent) "bIsFoldingWings" bIsFoldingWings))

(event EventOnNewPlayerWeaponType
  (params (NewType :type EFortItemType%20Enum :default "WorldItem") (OldType :type EFortItemType%20Enum :default "WorldItem"))
  (|CallParentFunction NewType OldType) ; This node is disabled and will not be called.\nDrag off pins to build functionality.
)

(event EventOnWeaponChanged
  (params (NewWeapon :type Fort%20Weapon%20Object%20Reference) (OldWeapon :type Fort%20Weapon%20Object%20Reference)))

(event EventOnWeaponFired)

(event EventOnNewWeapons
  (params (WeaponsAsItems :type Array%20of%20Fort%20Item%20Object%20References)))

(event EventOnCurieStateAttached
  (params (Tag :type Gameplay%20Tag%20Structure%20%28by%20ref%29)))

(event EventOnCurieStateDetached
  (params (Tag :type Gameplay%20Tag%20Structure%20%28by%20ref%29)))

(event EventOnBeginSkydiving
  (|CallParentFunction) ; This node is disabled and will not be called.\nDrag off pins to build functionality.
)

(event EventOnEndSkydiving
  (|CallParentFunction) ; This node is disabled and will not be called.\nDrag off pins to build functionality.
)

(event EventOnEnterVehicle
  (params (VehicleEntered :type Fort%20Vehicle%20Interface%20Interface%20%28by%20ref%29))
  (|CallParentFunction VehicleEntered) ; This node is disabled and will not be called.\nDrag off pins to build functionality.
)

(event EventOnExitVehicle
  (|CallParentFunction) ; This node is disabled and will not be called.\nDrag off pins to build functionality.
)

(event EventOnTargetingOrFirstPersonCameraChanged)

(event EventOnPawnJumped
  (bind _returnvalue (CallFunction|GetIdleVFXComponent))
  (|CallParentFunction)
  (Effects|Components|ParticleSystem|SetBoolParameter _returnvalue "bPawnJumped" true)
  (Utilities|FlowControl|Delay 0.5)
  (Effects|Components|ParticleSystem|SetBoolParameter _returnvalue "bPawnJumped" false))

(fn ConstructionScript
  (|CallParentFunction))
