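; Part modifier for the BisonDrain reactive back bling, driving its reactive material on the pawn's backpack skeletal mesh.
; After post-res-in, if frontend reactivity should display, it flags the materials as frontend and plays the ReactiveFrontEnd timeline into the alpha and enable parameters.
; Otherwise it clears the frontend flag and sets StartTime to a negative value.
; Any stat change stamps StartTime with current game time, leaving the animation to the material.
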(event EventOnPostResIn
  (bind _returnvalue (CustomCharacter|GetSkeletalMeshForPartOnPlayerPawn))
  (bind _returnvalue_2 (CustomCharacter|GetSkeletalMeshForPartOnPlayerPawn))
  (|CallParentFunction)
  (if (Reactive|Frontend|ShouldDisplayFrontendReactivity 0 "Backpack")
    (Rendering|Material|SetScalarParameterValueonMaterials _returnvalue "IsInFrontEnd" 1.0)
    (bind (_direction _reactivealpha _enablereactive) (Timeline|ReactiveFrontEnd Play)
      (:Update
        (Rendering|Material|SetScalarParameterValueonMaterials _returnvalue "FrontendAlpha" _reactivealpha)
        (Rendering|Material|SetScalarParameterValueonMaterials _returnvalue "ReactivityEnable" _enablereactive)))
    (else
      (Rendering|Material|SetScalarParameterValueonMaterials _returnvalue_2 "IsInFrontEnd" 0.0)
      (Rendering|Material|SetScalarParameterValueonMaterials _returnvalue_2 "StartTime" -1e+01))))

(event EventOnDayPhaseChanged
  (params (CurrentDayPhase :type EFortDayPhase%20Enum :default "Morning") (PreviousDayPhase :type EFortDayPhase%20Enum :default "Morning") (bAtCreation :type Boolean)))

(event EventOnNewPlayerWeaponType
  (params (NewType :type EFortItemType%20Enum :default "WorldItem") (OldType :type EFortItemType%20Enum :default "WorldItem"))
  (|CallParentFunction NewType OldType) ; This node is disabled and will not be called.\nDrag off pins to build functionality.
)

(event EventOnStatChanged
  (params (State%20Name :type Name :default "None") (Stat%20Value :type Integer))
  (|CallParentFunction State%20Name Stat%20Value)
  (Rendering|Material|SetScalarParameterValueonMaterials (CustomCharacter|GetSkeletalMeshForPartOnPlayerPawn) "StartTime" (Utilities|Time|GetGameTimeinSeconds)))

(event EventOnWeaponChanged
  (params (NewWeapon :type Fort%20Weapon%20Object%20Reference) (OldWeapon :type Fort%20Weapon%20Object%20Reference))
  (|CallParentFunction NewWeapon OldWeapon) ; This node is disabled and will not be called.\nDrag off pins to build functionality.
)

(event EventOnWeaponFired)

(event EventOnNewWeapons
  (params (WeaponsAsItems :type Array%20of%20Fort%20Item%20Object%20References)))

(event EventOnCurieStateAttached
  (params (Tag :type Gameplay%20Tag%20Structure%20%28by%20ref%29)))

(event EventOnCurieStateDetached
  (params (Tag :type Gameplay%20Tag%20Structure%20%28by%20ref%29)))

(event EventOnBeginSkydiving
  (|CallParentFunction) ; This node is disabled and will not be called.\nDrag off pins to build functionality.
)

(event EventOnEndSkydiving
  (|CallParentFunction) ; This node is disabled and will not be called.\nDrag off pins to build functionality.
)

(event EventOnEnterVehicle
  (params (VehicleEntered :type Fort%20Vehicle%20Interface%20Interface%20%28by%20ref%29))
  (|CallParentFunction VehicleEntered) ; This node is disabled and will not be called.\nDrag off pins to build functionality.
)

(event EventOnExitVehicle
  (|CallParentFunction) ; This node is disabled and will not be called.\nDrag off pins to build functionality.
)

(event EventOnTargetingOrFirstPersonCameraChanged)

(event EventOnHitPawn
  (params (SelfActor :type Actor%20Object%20Reference) (OtherActor :type Actor%20Object%20Reference) (NormalImpulse :type Vector :default "0, 0, 0") (Hit :type Hit%20Result%20Structure%20%28by%20ref%29)))

(fn ConstructionScript
  (|CallParentFunction))
