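; Backpack part modifier for the BlueBadass cosmetic that changes its material scalar parameters as cosmetic stages activate.
; The stage-activated event calls its parent, then switches on the stage number to one of thirteen custom events, printing an error for unhandled stages.
; Each event, gated by CanExecuteCosmeticEvents, iterates its stage's name-to-scalar map and sets those values on the pawn's skeletal mesh for that part type, flagging None keys.
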(event Custom|Stage%201
  (bind _stage_1_material_scalars (Variables|Default|GetStage%201%20Material%20Scalars))
  ;[1 1 1] STAGE 1
  (Networking|CanExecuteCosmeticEvents
    (:True
      (bind _keys (Utilities|Map|Keys _stage_1_material_scalars))
      (if (CallFunction|ArrayIsValid _keys)
        (for _array_element _keys
          (if (Utilities|Name|Equal(Name) _array_element "None")
            (Development|PrintString "What are you thinking?  A Variable called None.  Really?!")
            (else
              (Rendering|Material|SetScalarParameterValueonMaterials (Customization|GetSkeletalMeshforPartType (CustomCharacter|GetAssociatedPlayerPawn) (CustomCharacter|GetCharacterPartType) :!class /Script/FortniteGame.FortPlayerPawn) _array_element (Utilities|Map|Find _stage_1_material_scalars _array_element)))))))))

(event Custom|Stage%202
  (bind _stage_2_material_scalars (Variables|Default|GetStage%202%20Material%20Scalars))
  ;[1 1 1] STAGE 2
  (Networking|CanExecuteCosmeticEvents
    (:True
      (bind _keys (Utilities|Map|Keys _stage_2_material_scalars))
      (if (CallFunction|ArrayIsValid _keys)
        (for _array_element _keys
          (if (Utilities|Name|Equal(Name) _array_element "None")
            (Development|PrintString "What are you thinking?  A Variable called None.  Really?!")
            (else
              (Rendering|Material|SetScalarParameterValueonMaterials (Customization|GetSkeletalMeshforPartType (CustomCharacter|GetAssociatedPlayerPawn) (CustomCharacter|GetCharacterPartType) :!class /Script/FortniteGame.FortPlayerPawn) _array_element (Utilities|Map|Find _stage_2_material_scalars _array_element)))))))))

(event Custom|Stage%203
  (bind _stage_3_material_scalars (Variables|Default|GetStage%203%20Material%20Scalars))
  ;[1 1 1] STAGE 3
  (Networking|CanExecuteCosmeticEvents
    (:True
      (bind _keys (Utilities|Map|Keys _stage_3_material_scalars))
      (if (CallFunction|ArrayIsValid _keys)
        (for _array_element _keys
          (if (Utilities|Name|Equal(Name) _array_element "None")
            (Development|PrintString "What are you thinking?  A Variable called None.  Really?!")
            (else
              (Rendering|Material|SetScalarParameterValueonMaterials (Customization|GetSkeletalMeshforPartType (CustomCharacter|GetAssociatedPlayerPawn) (CustomCharacter|GetCharacterPartType) :!class /Script/FortniteGame.FortPlayerPawn) _array_element (Utilities|Map|Find _stage_3_material_scalars _array_element)))))))))

(event Custom|Stage%204
  (bind _stage_4_material_scalars (Variables|Default|GetStage%204%20Material%20Scalars))
  ;[1 1 1] STAGE 4
  (Networking|CanExecuteCosmeticEvents
    (:True
      (bind _keys (Utilities|Map|Keys _stage_4_material_scalars))
      (if (CallFunction|ArrayIsValid _keys)
        (for _array_element _keys
          (if (Utilities|Name|Equal(Name) _array_element "None")
            (Development|PrintString "What are you thinking?  A Variable called None.  Really?!")
            (else
              (Rendering|Material|SetScalarParameterValueonMaterials (Customization|GetSkeletalMeshforPartType (CustomCharacter|GetAssociatedPlayerPawn) (CustomCharacter|GetCharacterPartType) :!class /Script/FortniteGame.FortPlayerPawn) _array_element (Utilities|Map|Find _stage_4_material_scalars _array_element)))))))))

(event Custom|Stage%205
  (bind _stage_5_material_scalars (Variables|Default|GetStage%205%20Material%20Scalars))
  ;[1 1 1] STAGE 5
  (Networking|CanExecuteCosmeticEvents
    (:True
      (bind _keys (Utilities|Map|Keys _stage_5_material_scalars))
      (if (CallFunction|ArrayIsValid _keys)
        (for _array_element _keys
          (if (Utilities|Name|Equal(Name) _array_element "None")
            (Development|PrintString "What are you thinking?  A Variable called None.  Really?!")
            (else
              (Rendering|Material|SetScalarParameterValueonMaterials (Customization|GetSkeletalMeshforPartType (CustomCharacter|GetAssociatedPlayerPawn) (CustomCharacter|GetCharacterPartType) :!class /Script/FortniteGame.FortPlayerPawn) _array_element (Utilities|Map|Find _stage_5_material_scalars _array_element)))))))))

(event Custom|Stage%206
  (bind _stage_6_material_scalars (Variables|Default|GetStage%206%20Material%20Scalars))
  ;[1 1 1] STAGE 6
  (Networking|CanExecuteCosmeticEvents
    (:True
      (bind _keys (Utilities|Map|Keys _stage_6_material_scalars))
      (if (CallFunction|ArrayIsValid _keys)
        (for _array_element _keys
          (if (Utilities|Name|Equal(Name) _array_element "None")
            (Development|PrintString "What are you thinking?  A Variable called None.  Really?!")
            (else
              (Rendering|Material|SetScalarParameterValueonMaterials (Customization|GetSkeletalMeshforPartType (CustomCharacter|GetAssociatedPlayerPawn) (CustomCharacter|GetCharacterPartType) :!class /Script/FortniteGame.FortPlayerPawn) _array_element (Utilities|Map|Find _stage_6_material_scalars _array_element)))))))))

(event EventOnCosmeticStateActivated
  (params (Stat%20Name :type Name :default "None") (Stat%20Value :type Integer) (STAGE%20Num%20Being%20Activated :type Integer))
  (|CallParentFunction Stat%20Name Stat%20Value STAGE%20Num%20Being%20Activated)
  (switch int STAGE%20Num%20Being%20Activated
    (:Default
      (Development|PrintString (Utilities|String|Append "Unhandled Cosmetic Stage " (Utilities|String|ToString(Integer) STAGE%20Num%20Being%20Activated)) true true "(R=1.000000,G=0.000000,B=0.006820,A=1.000000)"))
    (:1
      (CallFunction|Stage1))
    (:2
      (CallFunction|Stage2))
    (:3
      (CallFunction|Stage3))
    (:4
      (CallFunction|Stage4))
    (:5
      (CallFunction|Stage5))
    (:6
      (CallFunction|Stage6))
    (:7
      (CallFunction|Stage7))
    (:8
      (CallFunction|Stage8))
    (:9
      (CallFunction|Stage9))
    (:10
      (CallFunction|Stage10))
    (:11
      (CallFunction|Stage11))
    (:12
      (CallFunction|Stage12))
    (:13
      (CallFunction|Stage13))))

(event Custom|Stage%207
  (bind _stage_7_material_scalars (Variables|Default|GetStage%207%20Material%20Scalars))
  ;[1 1 1] STAGE 7
  (Networking|CanExecuteCosmeticEvents
    (:True
      (bind _keys (Utilities|Map|Keys _stage_7_material_scalars))
      (if (CallFunction|ArrayIsValid _keys)
        (for _array_element _keys
          (if (Utilities|Name|Equal(Name) _array_element "None")
            (Development|PrintString "What are you thinking?  A Variable called None.  Really?!")
            (else
              (Rendering|Material|SetScalarParameterValueonMaterials (Customization|GetSkeletalMeshforPartType (CustomCharacter|GetAssociatedPlayerPawn) (CustomCharacter|GetCharacterPartType) :!class /Script/FortniteGame.FortPlayerPawn) _array_element (Utilities|Map|Find _stage_7_material_scalars _array_element)))))))))

(event Custom|Stage%208
  (bind _stage_8_material_scalars (Variables|Default|GetStage%208%20Material%20Scalars))
  ;[1 1 1] STAGE 8
  (Networking|CanExecuteCosmeticEvents
    (:True
      (bind _keys (Utilities|Map|Keys _stage_8_material_scalars))
      (if (CallFunction|ArrayIsValid _keys)
        (for _array_element _keys
          (if (Utilities|Name|Equal(Name) _array_element "None")
            (Development|PrintString "What are you thinking?  A Variable called None.  Really?!")
            (else
              (Rendering|Material|SetScalarParameterValueonMaterials (Customization|GetSkeletalMeshforPartType (CustomCharacter|GetAssociatedPlayerPawn) (CustomCharacter|GetCharacterPartType) :!class /Script/FortniteGame.FortPlayerPawn) _array_element (Utilities|Map|Find _stage_8_material_scalars _array_element)))))))))

(event Custom|Stage%209
  (bind _stage_9_material_scalars (Variables|Default|GetStage%209%20Material%20Scalars))
  ;[1 1 1] STAGE 9
  (Networking|CanExecuteCosmeticEvents
    (:True
      (bind _keys (Utilities|Map|Keys _stage_9_material_scalars))
      (if (CallFunction|ArrayIsValid _keys)
        (for _array_element _keys
          (if (Utilities|Name|Equal(Name) _array_element "None")
            (Development|PrintString "What are you thinking?  A Variable called None.  Really?!")
            (else
              (Rendering|Material|SetScalarParameterValueonMaterials (Customization|GetSkeletalMeshforPartType (CustomCharacter|GetAssociatedPlayerPawn) (CustomCharacter|GetCharacterPartType) :!class /Script/FortniteGame.FortPlayerPawn) _array_element (Utilities|Map|Find _stage_9_material_scalars _array_element)))))))))

(event Custom|Stage%2010
  (bind _stage_10_material_scalars (Variables|Default|GetStage%2010%20Material%20Scalars))
  ;[1 1 1] STAGE 10
  (Networking|CanExecuteCosmeticEvents
    (:True
      (bind _keys (Utilities|Map|Keys _stage_10_material_scalars))
      (if (CallFunction|ArrayIsValid _keys)
        (for _array_element _keys
          (if (Utilities|Name|Equal(Name) _array_element "None")
            (Development|PrintString "What are you thinking?  A Variable called None.  Really?!")
            (else
              (Rendering|Material|SetScalarParameterValueonMaterials (Customization|GetSkeletalMeshforPartType (CustomCharacter|GetAssociatedPlayerPawn) (CustomCharacter|GetCharacterPartType) :!class /Script/FortniteGame.FortPlayerPawn) _array_element (Utilities|Map|Find _stage_10_material_scalars _array_element)))))))))

(event Custom|Stage%2011
  (bind _stage_11_material_scalars (Variables|Default|GetStage%2011%20Material%20Scalars))
  ;[1 1 1] STAGE 11
  (Networking|CanExecuteCosmeticEvents
    (:True
      (bind _keys (Utilities|Map|Keys _stage_11_material_scalars))
      (if (CallFunction|ArrayIsValid _keys)
        (for _array_element _keys
          (if (Utilities|Name|Equal(Name) _array_element "None")
            (Development|PrintString "What are you thinking?  A Variable called None.  Really?!")
            (else
              (Rendering|Material|SetScalarParameterValueonMaterials (Customization|GetSkeletalMeshforPartType (CustomCharacter|GetAssociatedPlayerPawn) (CustomCharacter|GetCharacterPartType) :!class /Script/FortniteGame.FortPlayerPawn) _array_element (Utilities|Map|Find _stage_11_material_scalars _array_element)))))))))

(event Custom|Stage%2012
  (bind _stage_12_material_scalars (Variables|Default|GetStage%2012%20Material%20Scalars))
  ;[1 1 1] STAGE 12
  (Networking|CanExecuteCosmeticEvents
    (:True
      (bind _keys (Utilities|Map|Keys _stage_12_material_scalars))
      (if (CallFunction|ArrayIsValid _keys)
        (for _array_element _keys
          (if (Utilities|Name|Equal(Name) _array_element "None")
            (Development|PrintString "What are you thinking?  A Variable called None.  Really?!")
            (else
              (Rendering|Material|SetScalarParameterValueonMaterials (Customization|GetSkeletalMeshforPartType (CustomCharacter|GetAssociatedPlayerPawn) (CustomCharacter|GetCharacterPartType) :!class /Script/FortniteGame.FortPlayerPawn) _array_element (Utilities|Map|Find _stage_12_material_scalars _array_element)))))))))

(event Custom|Stage%2013
  (bind _stage_13_material_scalars (Variables|Default|GetStage%2013%20Material%20Scalars))
  ;[1 1 1] STAGE 13
  (Networking|CanExecuteCosmeticEvents
    (:True
      (bind _keys (Utilities|Map|Keys _stage_13_material_scalars))
      (if (CallFunction|ArrayIsValid _keys)
        (for _array_element _keys
          (if (Utilities|Name|Equal(Name) _array_element "None")
            (Development|PrintString "What are you thinking?  A Variable called None.  Really?!")
            (else
              (Rendering|Material|SetScalarParameterValueonMaterials (Customization|GetSkeletalMeshforPartType (CustomCharacter|GetAssociatedPlayerPawn) (CustomCharacter|GetCharacterPartType) :!class /Script/FortniteGame.FortPlayerPawn) _array_element (Utilities|Map|Find _stage_13_material_scalars _array_element)))))))))

(fn ArrayIsValid
  (params (Name%20Array :type Array%20of%20Names))
  (return (> (Utilities|Array|Length Name%20Array) 0)))

(fn ConstructionScript
  (|CallParentFunction))
