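; Part modifier for the BlueBadass male backpack that progressively lights up the cosmetic across 13 activated stages.
; Each stage event, gated to cosmetic-capable clients, writes its name-to-scalar map onto the part's mesh materials.
; Stages 1-11 also advance LightRays "Progression"; stages 1-12 adjust three point lights, moving through yellow-orange, pink-purple and blue-cyan.
; On mobile, point-light updates are disabled and LightA alone is destroyed.
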(event Custom|Stage%201
  (bind _stage_1_material_scalars (Variables|Default|GetStage%201%20Material%20Scalars))
  (bind _lighta (Variables|Default|GetLightA))
  ;[1 1 1] STAGE 1
  (Networking|CanExecuteCosmeticEvents
    (:True
      (Utilities|FlowControl|Sequence
        (:then_0
          (bind _keys (Utilities|Map|Keys _stage_1_material_scalars))
          (if (CallFunction|ArrayIsValid _keys)
            (for _array_element _keys
              (if (Utilities|Name|Equal(Name) _array_element "None")
                (Development|PrintString "What are you thinking?  A Variable called None.  Really?!")
                (else
                  (Rendering|Material|SetScalarParameterValueonMaterials (CustomCharacter|GetSkeletalMeshForPartOnPlayerPawn) _array_element (Utilities|Map|Find _stage_1_material_scalars _array_element)))))))
        (:then_1
          (Rendering|Material|SetScalarParameterValueonMaterials (Variables|Default|GetLightRays) "Progression" (Variables|Default|GetStage1_Light))
          (if (Variables|Default|GetPointLightsActive)
            (Rendering|Components|Light|SetIntensity _lighta 1.0)
            (Rendering|Components|Light|SetLightColor _lighta "(R=0.799103,G=0.779854,B=0.000000,A=1.000000)" true)
            (Rendering|Components|Light|SetIntensity (Variables|Default|GetLightB) 0.0)))))))

(event Custom|Stage%202
  (bind _stage_2_material_scalars (Variables|Default|GetStage%202%20Material%20Scalars))
  (bind _lighta (Variables|Default|GetLightA))
  ;[1 1 1] STAGE 2
  (Networking|CanExecuteCosmeticEvents
    (:True
      (Utilities|FlowControl|Sequence
        (:then_0
          (bind _keys (Utilities|Map|Keys _stage_2_material_scalars))
          (if (CallFunction|ArrayIsValid _keys)
            (for _array_element _keys
              (if (Utilities|Name|Equal(Name) _array_element "None")
                (Development|PrintString "What are you thinking?  A Variable called None.  Really?!")
                (else
                  (Rendering|Material|SetScalarParameterValueonMaterials (CustomCharacter|GetSkeletalMeshForPartOnPlayerPawn) _array_element (Utilities|Map|Find _stage_2_material_scalars _array_element)))))))
        (:then_1
          (Rendering|Material|SetScalarParameterValueonMaterials (Variables|Default|GetLightRays) "Progression" (Variables|Default|GetStage2_Light))
          (if (Variables|Default|GetPointLightsActive)
            (Rendering|Components|Light|SetIntensity _lighta 2.0)
            (Rendering|Components|Light|SetLightColor _lighta "(R=0.930111,G=0.713085,B=0.000000,A=1.000000)" true)
            (Rendering|Components|Light|SetIntensity (Variables|Default|GetLightB) 0.0)))))))

(event Custom|Stage%203
  (bind _stage_3_material_scalars (Variables|Default|GetStage%203%20Material%20Scalars))
  (bind _lighta (Variables|Default|GetLightA))
  ;[1 1 1] STAGE 3
  (Networking|CanExecuteCosmeticEvents
    (:True
      (Utilities|FlowControl|Sequence
        (:then_0
          (bind _keys (Utilities|Map|Keys _stage_3_material_scalars))
          (if (CallFunction|ArrayIsValid _keys)
            (for _array_element _keys
              (if (Utilities|Name|Equal(Name) _array_element "None")
                (Development|PrintString "What are you thinking?  A Variable called None.  Really?!")
                (else
                  (Rendering|Material|SetScalarParameterValueonMaterials (CustomCharacter|GetSkeletalMeshForPartOnPlayerPawn) _array_element (Utilities|Map|Find _stage_3_material_scalars _array_element)))))))
        (:then_1
          (Rendering|Material|SetScalarParameterValueonMaterials (Variables|Default|GetLightRays) "Progression" (Variables|Default|GetStage3_Light))
          (if (Variables|Default|GetPointLightsActive)
            (Rendering|Components|Light|SetIntensity _lighta 3.0)
            (Rendering|Components|Light|SetLightColor _lighta "(R=0.930111,G=0.448200,B=0.000000,A=1.000000)" true)
            (Rendering|Components|Light|SetIntensity (Variables|Default|GetLightB) 0.0)))))))

(event EventBeginPlay
  (bind _returnvalue (TouchPlatform|IsMobileGame))
  (|CallParentFunction)
  (Variables|Default|SetPointLightsActive (not _returnvalue))
  (if _returnvalue
    (Components|DestroyComponent (Variables|Default|GetLightA))))

(event EventOnPostResIn
  (|CallParentFunction)
  (Transformation|AttachComponentToComponent (Variables|Default|GetLightRays) (CustomCharacter|GetSkeletalMeshForPartOnPlayerPawn) "root" "SnapToTarget" "SnapToTarget" "KeepRelative" true)
  (Rendering|Material|SetScalarParameterValueonMaterials (Variables|Default|GetLightRays) "Progression" 0.0))

(event Custom|Stage%204
  (bind _lighta (Variables|Default|GetLightA))
  (bind _stage_4_material_scalars (Variables|Default|GetStage%204%20Material%20Scalars))
  ;[1 1 1] STAGE 4
  (Networking|CanExecuteCosmeticEvents
    (:True
      (Utilities|FlowControl|Sequence
        (:then_0
          (bind _keys (Utilities|Map|Keys _stage_4_material_scalars))
          (if (CallFunction|ArrayIsValid _keys)
            (for _array_element _keys
              (if (Utilities|Name|Equal(Name) _array_element "None")
                (Development|PrintString "What are you thinking?  A Variable called None.  Really?!")
                (else
                  (Rendering|Material|SetScalarParameterValueonMaterials (CustomCharacter|GetSkeletalMeshForPartOnPlayerPawn) _array_element (Utilities|Map|Find _stage_4_material_scalars _array_element)))))))
        (:then_1
          (Rendering|Material|SetScalarParameterValueonMaterials (Variables|Default|GetLightRays) "Progression" (Variables|Default|GetStage4_Light))
          (if (Variables|Default|GetPointLightsActive)
            (Rendering|Components|Light|SetIntensity _lighta 4.0)
            (Rendering|Components|Light|SetLightColor _lighta "(R=0.921582,G=0.460791,B=0.000000,A=1.000000)" true)
            (Rendering|Components|Light|SetIntensity (Variables|Default|GetLightB) 0.0)))))))

(event EventOnCosmeticStateActivated
  (params (Stat%20Name :type Name :default "None") (Stat%20Value :type Integer) (STAGE%20Num%20Being%20Activated :type Integer))
  (|CallParentFunction Stat%20Name Stat%20Value STAGE%20Num%20Being%20Activated)
  (switch int STAGE%20Num%20Being%20Activated
    (:Default
      (Development|PrintString (Utilities|String|Append "Unhandled Cosmetic Stage " (Utilities|String|ToString(Integer) STAGE%20Num%20Being%20Activated)) true true "(R=1.000000,G=0.000000,B=0.006820,A=1.000000)"))
    (:1
      (CallFunction|Stage1))
    (:2
      (CallFunction|Stage2))
    (:3
      (CallFunction|Stage3))
    (:4
      (CallFunction|Stage4))
    (:5
      (CallFunction|Stage5))
    (:6
      (CallFunction|Stage6))
    (:7
      (CallFunction|Stage7))
    (:8
      (CallFunction|Stage8))
    (:9
      (CallFunction|Stage9))
    (:10
      (CallFunction|Stage10))
    (:11
      (CallFunction|Stage11))
    (:12
      (CallFunction|Stage12))
    (:13
      (CallFunction|Stage13))))

(event Custom|Stage%205
  (bind _stage_5_material_scalars (Variables|Default|GetStage%205%20Material%20Scalars))
  (bind _lightb (Variables|Default|GetLightB))
  ;[1 1 1] STAGE 5
  (Networking|CanExecuteCosmeticEvents
    (:True
      (Utilities|FlowControl|Sequence
        (:then_0
          (bind _keys (Utilities|Map|Keys _stage_5_material_scalars))
          (if (CallFunction|ArrayIsValid _keys)
            (for _array_element _keys
              (if (Utilities|Name|Equal(Name) _array_element "None")
                (Development|PrintString "What are you thinking?  A Variable called None.  Really?!")
                (else
                  (Rendering|Material|SetScalarParameterValueonMaterials (CustomCharacter|GetSkeletalMeshForPartOnPlayerPawn) _array_element (Utilities|Map|Find _stage_5_material_scalars _array_element)))))))
        (:then_1
          (Rendering|Material|SetScalarParameterValueonMaterials (Variables|Default|GetLightRays) "Progression" (Variables|Default|GetStage5_Light))
          (if (Variables|Default|GetPointLightsActive)
            (Rendering|Components|Light|SetIntensity _lightb 1.0)
            (Rendering|Components|Light|SetLightColor _lightb "(R=1.000000,G=0.133929,B=0.706682,A=1.000000)" true)
            (Rendering|Components|Light|SetIntensity (Variables|Default|GetLightC) 0.0)))))))

(event Custom|Stage%206
  (bind _lightb (Variables|Default|GetLightB))
  (bind _stage_6_material_scalars (Variables|Default|GetStage%206%20Material%20Scalars))
  ;[1 1 1] STAGE 6
  (Networking|CanExecuteCosmeticEvents
    (:True
      (Utilities|FlowControl|Sequence
        (:then_0
          (bind _keys (Utilities|Map|Keys _stage_6_material_scalars))
          (if (CallFunction|ArrayIsValid _keys)
            (for _array_element _keys
              (if (Utilities|Name|Equal(Name) _array_element "None")
                (Development|PrintString "What are you thinking?  A Variable called None.  Really?!")
                (else
                  (Rendering|Material|SetScalarParameterValueonMaterials (CustomCharacter|GetSkeletalMeshForPartOnPlayerPawn) _array_element (Utilities|Map|Find _stage_6_material_scalars _array_element)))))))
        (:then_1
          (Rendering|Material|SetScalarParameterValueonMaterials (Variables|Default|GetLightRays) "Progression" (Variables|Default|GetStage6_Light))
          (if (Variables|Default|GetPointLightsActive)
            (Rendering|Components|Light|SetIntensity _lightb 2.5)
            (Rendering|Components|Light|SetLightColor _lightb "(R=0.495552,G=0.000000,B=0.896269,A=1.000000)" true)
            (Rendering|Components|Light|SetIntensity (Variables|Default|GetLightC) 0.0)))))))

(event Custom|Stage%207
  (bind _stage_7_material_scalars (Variables|Default|GetStage%207%20Material%20Scalars))
  (bind _lightb (Variables|Default|GetLightB))
  ;[1 1 1] STAGE 7
  (Networking|CanExecuteCosmeticEvents
    (:True
      (Utilities|FlowControl|Sequence
        (:then_0
          (bind _keys (Utilities|Map|Keys _stage_7_material_scalars))
          (if (CallFunction|ArrayIsValid _keys)
            (for _array_element _keys
              (if (Utilities|Name|Equal(Name) _array_element "None")
                (Development|PrintString "What are you thinking?  A Variable called None.  Really?!")
                (else
                  (Rendering|Material|SetScalarParameterValueonMaterials (CustomCharacter|GetSkeletalMeshForPartOnPlayerPawn) _array_element (Utilities|Map|Find _stage_7_material_scalars _array_element)))))))
        (:then_1
          (Rendering|Material|SetScalarParameterValueonMaterials (Variables|Default|GetLightRays) "Progression" (Variables|Default|GetStage7_Light))
          (if (Variables|Default|GetPointLightsActive)
            (Rendering|Components|Light|SetIntensity _lightb 4.0)
            (Rendering|Components|Light|SetLightColor _lightb "(R=0.284843,G=0.000000,B=1.000000,A=1.000000)" true)
            (Rendering|Components|Light|SetIntensity (Variables|Default|GetLightC) 0.0)))))))

(event Custom|Stage%208
  (bind _stage_8_material_scalars (Variables|Default|GetStage%208%20Material%20Scalars))
  (bind _lightc (Variables|Default|GetLightC))
  ;[1 1 1] STAGE 8
  (Networking|CanExecuteCosmeticEvents
    (:True
      (Utilities|FlowControl|Sequence
        (:then_0
          (bind _keys (Utilities|Map|Keys _stage_8_material_scalars))
          (if (CallFunction|ArrayIsValid _keys)
            (for _array_element _keys
              (if (Utilities|Name|Equal(Name) _array_element "None")
                (Development|PrintString "What are you thinking?  A Variable called None.  Really?!")
                (else
                  (Rendering|Material|SetScalarParameterValueonMaterials (CustomCharacter|GetSkeletalMeshForPartOnPlayerPawn) _array_element (Utilities|Map|Find _stage_8_material_scalars _array_element)))))))
        (:then_1
          (Rendering|Material|SetScalarParameterValueonMaterials (Variables|Default|GetLightRays) "Progression" (Variables|Default|GetStage8_Light))
          (if (Variables|Default|GetPointLightsActive)
            (Rendering|Components|Light|SetIntensity _lightc 1.0)
            (Rendering|Components|Light|SetLightColor _lightc "(R=0.000000,G=0.347129,B=0.982251,A=1.000000)" true)))))))

(event Custom|Stage%209
  (bind _stage_9_material_scalars (Variables|Default|GetStage%209%20Material%20Scalars))
  (bind _lightc (Variables|Default|GetLightC))
  ;[1 1 1] STAGE 9
  (Networking|CanExecuteCosmeticEvents
    (:True
      (Utilities|FlowControl|Sequence
        (:then_0
          (bind _keys (Utilities|Map|Keys _stage_9_material_scalars))
          (if (CallFunction|ArrayIsValid _keys)
            (for _array_element _keys
              (if (Utilities|Name|Equal(Name) _array_element "None")
                (Development|PrintString "What are you thinking?  A Variable called None.  Really?!")
                (else
                  (Rendering|Material|SetScalarParameterValueonMaterials (CustomCharacter|GetSkeletalMeshForPartOnPlayerPawn) _array_element (Utilities|Map|Find _stage_9_material_scalars _array_element)))))))
        (:then_1
          (Rendering|Material|SetScalarParameterValueonMaterials (Variables|Default|GetLightRays) "Progression" (Variables|Default|GetStage9_Light))
          (if (Variables|Default|GetPointLightsActive)
            (Rendering|Components|Light|SetIntensity _lightc 2.0)
            (Rendering|Components|Light|SetLightColor _lightc "(R=0.000000,G=0.791761,B=0.930111,A=1.000000)" true)))))))

(event Custom|Stage%2010
  (bind _stage_10_material_scalars (Variables|Default|GetStage%2010%20Material%20Scalars))
  (bind _lightc (Variables|Default|GetLightC))
  ;[1 1 1] STAGE 10
  (Networking|CanExecuteCosmeticEvents
    (:True
      (Utilities|FlowControl|Sequence
        (:then_0
          (bind _keys (Utilities|Map|Keys _stage_10_material_scalars))
          (if (CallFunction|ArrayIsValid _keys)
            (for _array_element _keys
              (if (Utilities|Name|Equal(Name) _array_element "None")
                (Development|PrintString "What are you thinking?  A Variable called None.  Really?!")
                (else
                  (Rendering|Material|SetScalarParameterValueonMaterials (CustomCharacter|GetSkeletalMeshForPartOnPlayerPawn) _array_element (Utilities|Map|Find _stage_10_material_scalars _array_element)))))))
        (:then_1
          (Rendering|Material|SetScalarParameterValueonMaterials (Variables|Default|GetLightRays) "Progression" (Variables|Default|GetStage10_Light))
          (if (Variables|Default|GetPointLightsActive)
            (Rendering|Components|Light|SetIntensity _lightc 3.0)
            (Rendering|Components|Light|SetLightColor _lightc "(R=0.000000,G=0.856517,B=0.904661,A=1.000000)" true)))))))

(event Custom|Stage%2011
  (bind _stage_11_material_scalars (Variables|Default|GetStage%2011%20Material%20Scalars))
  (bind _lightc (Variables|Default|GetLightC))
  ;[1 1 1] STAGE 11
  (Networking|CanExecuteCosmeticEvents
    (:True
      (Utilities|FlowControl|Sequence
        (:then_0
          (bind _keys (Utilities|Map|Keys _stage_11_material_scalars))
          (if (CallFunction|ArrayIsValid _keys)
            (for _array_element _keys
              (if (Utilities|Name|Equal(Name) _array_element "None")
                (Development|PrintString "What are you thinking?  A Variable called None.  Really?!")
                (else
                  (Rendering|Material|SetScalarParameterValueonMaterials (CustomCharacter|GetSkeletalMeshForPartOnPlayerPawn) _array_element (Utilities|Map|Find _stage_11_material_scalars _array_element)))))))
        (:then_1
          (Rendering|Material|SetScalarParameterValueonMaterials (Variables|Default|GetLightRays) "Progression" (Variables|Default|GetStage11_Light))
          (if (Variables|Default|GetPointLightsActive)
            (Rendering|Components|Light|SetIntensity _lightc 4.0)
            (Rendering|Components|Light|SetLightColor _lightc "(R=0.000000,G=0.856517,B=0.904661,A=1.000000)" true)
            (Rendering|Components|Light|SetIntensity (Variables|Default|GetLightB) 4.0)))))))

(event Custom|Stage%2012
  (bind _stage_12_material_scalars (Variables|Default|GetStage%2012%20Material%20Scalars))
  ;[1 1 1] STAGE 12
  (Networking|CanExecuteCosmeticEvents
    (:True
      (Utilities|FlowControl|Sequence
        (:then_0
          (bind _keys (Utilities|Map|Keys _stage_12_material_scalars))
          (if (CallFunction|ArrayIsValid _keys)
            (for _array_element _keys
              (if (Utilities|Name|Equal(Name) _array_element "None")
                (Development|PrintString "What are you thinking?  A Variable called None.  Really?!")
                (else
                  (Rendering|Material|SetScalarParameterValueonMaterials (CustomCharacter|GetSkeletalMeshForPartOnPlayerPawn) _array_element (Utilities|Map|Find _stage_12_material_scalars _array_element)))))))
        (:then_1
          (if (Variables|Default|GetPointLightsActive)
            (Rendering|Components|Light|SetIntensity (Variables|Default|GetLightA) 6.0)))))))

(event Custom|Stage%2013
  (bind _stage_13_material_scalars (Variables|Default|GetStage%2013%20Material%20Scalars))
  ;[1 1 1] STAGE 13
  (Networking|CanExecuteCosmeticEvents
    (:True
      (bind _keys (Utilities|Map|Keys _stage_13_material_scalars))
      (if (CallFunction|ArrayIsValid _keys)
        (for _array_element _keys
          (if (Utilities|Name|Equal(Name) _array_element "None")
            (Development|PrintString "What are you thinking?  A Variable called None.  Really?!")
            (else
              (Rendering|Material|SetScalarParameterValueonMaterials (CustomCharacter|GetSkeletalMeshForPartOnPlayerPawn) _array_element (Utilities|Map|Find _stage_13_material_scalars _array_element)))))))))

(fn ArrayIsValid
  (params (Name%20Array :type Array%20of%20Names))
  (return (> (Utilities|Array|Length Name%20Array) 0)))

(fn ConstructionScript
  (|CallParentFunction))
