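; Back bling part modifier for the BrakePedal backpack that plays a Niagara effect when the player pawn lands.
; On the landed event it calls the parent, then spawns the backpack's Jump effect attached to the FX_Base socket on the skeletal mesh of the pawn's cosmetic part.
; The construction script only runs the parent logic.
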(event EventOnPawnLanded
  (|CallParentFunction)
  (PartModifier|SpawnSystem|SpawnSystemAttachedPartModifier "/BRCosmetics/Accessories/FORT_Backpacks/Backpack_M_MED_BrakePedal/FX/NS_Backpack_BrakePedal_Jump.NS_Backpack_BrakePedal_Jump" (CustomCharacter|GetSkeletalMeshForPartOnPlayerPawn) "FX_Base"))

(fn ConstructionScript
  (|CallParentFunction))
