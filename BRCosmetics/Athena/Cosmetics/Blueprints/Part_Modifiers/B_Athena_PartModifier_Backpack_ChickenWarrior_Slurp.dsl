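; Blueprint part modifier for the Chicken Warrior Slurp backpack cosmetic that makes its idle effect react when the pawn lands.
; On the landed event it calls the parent, then restarts a timeline from the beginning.
; Each timeline update writes the track value into the "Land" float parameter of the idle VFX particle component.
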(event EventOnPawnLanded
  (|CallParentFunction)
  (bind (_direction _spawntime) (Timeline|Timeline_0 PlayFromStart)
    (:Update
      (Effects|Components|ParticleSystem|SetFloatParameter (CallFunction|GetIdleVFXComponent) "Land" _spawntime))))

(fn ConstructionScript
  (|CallParentFunction))
