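; Part modifier for the ChicleVeil backpack cosmetic that adds a reactive effect on player kills and previews it in the frontend.
; When reactive, it watches the pawn's kill stat, spawns the kill FX system attached to the part's mesh and restarts a timeline driving an emissive boost material parameter.
; In the frontend, a looping timer replays the same effect and is cleared when reactivity preview should not display.
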(event Custom|FrontEnd_ReactivityPreview
  (bind _returnvalue (CustomCharacter|GetSkeletalMeshForPartOnPlayerPawn))
  (bind _killfxsystem (Variables|Default|GetKillFXSystem))
  ;[1 1 1] Displays effect on frontend
  (if (Reactive|Frontend|ShouldDisplayFrontendReactivity 0 "Backpack")
    (Utilities|FlowControl|Sequence
      (:then_0
        (Utilities|IsValid _killfxsystem
          (:Is%20Valid
            (PartModifier|SpawnSystem|SpawnSystemAttachedPartModifier _killfxsystem _returnvalue))))
      (:then_1
        (bind (_direction _boost) (Timeline|Timeline_1 PlayFromStart)
          (:Update
            (Rendering|Material|SetScalarParameterValueonMaterials _returnvalue "OnKillVibrantEmissiveAmount" _boost)))))
    (else
      (Utilities|Time|ClearandInvalidateTimerbyHandle (Variables|Default|GetFrontEndReactivityTimer)))))

(event EventOnPawnAndPartChanged
  (bind _returnvalue (CustomCharacter|GetSkeletalMeshForPartOnPlayerPawn))
  (bind _killfxsystem (Variables|Default|GetKillFXSystem))
  (|CallParentFunction)
  (Utilities|FlowControl|Sequence
    (:then_0
      (if (Variables|Default|GetIs%20Reactive)
        ;[1 1 1] On Kill Event
        (Cosmetics|WatchCosmeticStat (CustomCharacter|GetAssociatedPlayerPawn) "(TagName=\"StatManager.AthenaKills\")" 0 false
          (:OnStatChanged
            (Utilities|FlowControl|Sequence
              (:then_0
                (Utilities|IsValid _killfxsystem
                  (:Is%20Valid
                    (PartModifier|SpawnSystem|SpawnSystemAttachedPartModifier _killfxsystem _returnvalue))))
              (:then_1
                (Timeline|Timeline_1 PlayFromStart)))))))
    (:then_1
      ;[1 0.1636 0.04388] Frontend reactivity preview setup
      (if (Reactive|Frontend|ShouldDisplayFrontendReactivity 0 "Backpack")
        (bind _returnvalue_2 (Utilities|Time|SetTimerbyFunctionName 0 "FrontEnd_ReactivityPreview" 7.2 true))
        (Variables|Default|SetFrontEndReactivityTimer _returnvalue_2)
        (else
          (Utilities|Time|ClearandInvalidateTimerbyHandle (Variables|Default|GetFrontEndReactivityTimer)))))))

(fn ConstructionScript)
