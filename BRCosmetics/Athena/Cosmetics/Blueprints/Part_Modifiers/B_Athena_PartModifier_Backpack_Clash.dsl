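; Part modifier for the Clash back bling that steps its VFX and sound through three reactive stages as the player's kills rise.
; A watched AthenaKills stat triggers stages at one, two and three kills; in the frontend, a timeline previews them except on holograms.
; Each stage activates idle VFX, spawns a transition Niagara system and applies per-stage parameter maps to both.
; PlayAudio spawns an attached sound with a Stage parameter.
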(event EventOnPostResIn
  (bind _returnvalue (CustomCharacter|GetAssociatedPlayerPawn))
  (bind _bisinfrontendhologram (Variables|/Script/FortniteGame.FortPlayerPawn|GetbIsInFrontEndHologram _returnvalue))
  (|CallParentFunction)
  ;[0.6302 0.5146 0.4134] Frontend
  (if (Reactive|Frontend|ShouldDisplayFrontendReactivity 0 "Backpack")
    (Timeline|FrontEnd Play
      (:Stage1
        (if (not _bisinfrontendhologram)
          (CallFunction|Stage1)
          (CallFunction|PlayAudio 0)))
      (:Stage2
        (if (not _bisinfrontendhologram)
          (CallFunction|Stage2)
          (CallFunction|PlayAudio 1)))
      (:Stage3
        (if (not _bisinfrontendhologram)
          (CallFunction|Stage3)
          (CallFunction|PlayAudio 2))))))

(event EventOnCosmeticStateActivated
  (params (Stat%20Name :type Name :default "None") (Stat%20Value :type Integer) (STAGE%20Num%20Being%20Activated :type Integer))
  (|CallParentFunction Stat%20Name Stat%20Value STAGE%20Num%20Being%20Activated))

(event EventOnPawnAndPartChanged
  (|CallParentFunction)
  (bind (_statname _statvalue) (Cosmetics|WatchCosmeticStat (CustomCharacter|GetAssociatedPlayerPawn) "(TagName=\"StatManager.AthenaKills\")" 0 false)
    (:OnStatChanged
      ;[1 1 1] Comment
      (switch int (Math|Integer|Clamp(Integer) _statvalue 0 6)
        (:1
          (CallFunction|Stage1)
          (CallFunction|PlayAudio 0))
        (:2
          (CallFunction|Stage2)
          (CallFunction|PlayAudio 1))
        (:3
          (CallFunction|Stage3)
          (CallFunction|PlayAudio 2))))))

(event Custom|Stage%201
  (bind _returnvalue (CallFunction|GetIdleVFXComponent))
  (bind _ns_stage1_transition_float (Variables|Default|GetNS_Stage1_Transition_Float))
  (bind _transition_fx (Variables|Default|GetTransition%20FX))
  (bind _ns_stage1_idle_float (Variables|Default|GetNS_Stage1_Idle_Float))
  ;[1 1 1] Stage1
  (Components|Activation|Activate (CallFunction|GetIdleVFXComponent) true)
  (bind _returnvalue_2 (PartModifier|SpawnSystem|SpawnSystemAttachedPartModifier "/BRCosmetics/Accessories/FORT_Backpacks/Backpack_M_MED_Clash/FX/NS_Backpack_Clash_Transition.NS_Backpack_Clash_Transition" (CustomCharacter|GetSkeletalMeshForPartOnPlayerPawn) "Root" "0, 0, 0" "0, 0, 0" "SnapToTarget")
    (:Success
      (Variables|Default|SetTransition%20FX _returnvalue_2))
    (:Failure
      _))
  (Utilities|FlowControl|Sequence
    (:then_0
      (bind _keys (Utilities|Map|Keys _ns_stage1_idle_float))
      (bind _values (Utilities|Map|Values _ns_stage1_idle_float))
      (for (_array_element _array_index) _keys
        (Effects|Components|ParticleSystem|SetFloatParameter _returnvalue _array_element (Utilities|Array|Get(acopy) _values _array_index))))
    (:then_1
      (Effects|Components|ParticleSystem|SetIntParameter _returnvalue "BoneFlowSmallSpawn" 1))
    (:then_2
      (bind _keys_2 (Utilities|Map|Keys _ns_stage1_transition_float))
      (bind _values_2 (Utilities|Map|Values _ns_stage1_transition_float))
      (for (_array_element_2 _array_index_2) _keys_2
        (Effects|Components|ParticleSystem|SetFloatParameter (Variables|Default|GetTransition%20FX) _array_element_2 (Utilities|Array|Get(acopy) _values_2 _array_index_2))))
    (:then_3
      (Utilities|IsValid _transition_fx
        (:Is%20Valid
          (Effects|Components|ParticleSystem|SetIntParameter _transition_fx "BoneFlowPerGroup" 0))))))

(event Custom|Stage%202
  (bind _returnvalue (CallFunction|GetIdleVFXComponent))
  (bind _transition_fx (Variables|Default|GetTransition%20FX))
  (bind _ns_stage2_transition_float (Variables|Default|GetNS_Stage2_Transition_Float))
  (bind _ns_stage2_idle_float (Variables|Default|GetNS_Stage2_Idle_Float))
  ;[1 1 1] Stage 2
  (Components|Activation|Activate (CallFunction|GetIdleVFXComponent) true)
  (bind _returnvalue_2 (PartModifier|SpawnSystem|SpawnSystemAttachedPartModifier "/BRCosmetics/Accessories/FORT_Backpacks/Backpack_M_MED_Clash/FX/NS_Backpack_Clash_Transition.NS_Backpack_Clash_Transition" (CustomCharacter|GetSkeletalMeshForPartOnPlayerPawn) "Root" "0, 0, 0" "0, 0, 0" "SnapToTarget")
    (:Success
      (Variables|Default|SetTransition%20FX _returnvalue_2))
    (:Failure
      _))
  (Utilities|FlowControl|Sequence
    (:then_0
      (bind _keys (Utilities|Map|Keys _ns_stage2_idle_float))
      (bind _values (Utilities|Map|Values _ns_stage2_idle_float))
      (for (_array_element _array_index) _keys
        (Effects|Components|ParticleSystem|SetFloatParameter _returnvalue _array_element (Utilities|Array|Get(acopy) _values _array_index))))
    (:then_1
      (Effects|Components|ParticleSystem|SetIntParameter _returnvalue "BoneFlowSmallSpawn" 0))
    (:then_2
      (bind _keys_2 (Utilities|Map|Keys _ns_stage2_transition_float))
      (bind _values_2 (Utilities|Map|Values _ns_stage2_transition_float))
      (for (_array_element_2 _array_index_2) _keys_2
        (Effects|Components|ParticleSystem|SetFloatParameter (Variables|Default|GetTransition%20FX) _array_element_2 (Utilities|Array|Get(acopy) _values_2 _array_index_2))))
    (:then_3
      (Utilities|IsValid _transition_fx
        (:Is%20Valid
          (Effects|Components|ParticleSystem|SetIntParameter _transition_fx "BoneFlowPerGroup" 0))))))

(event Custom|Stage3
  (bind _returnvalue (CallFunction|GetIdleVFXComponent))
  (bind _transition_fx (Variables|Default|GetTransition%20FX))
  (bind _ns_stage3_transition_float (Variables|Default|GetNS_Stage3_Transition_Float))
  (bind _ns_stage3_idle_float (Variables|Default|GetNS_Stage3_Idle_Float))
  ;[1 1 1] Stage3
  (Components|Activation|Activate (CallFunction|GetIdleVFXComponent) true)
  (bind _returnvalue_2 (PartModifier|SpawnSystem|SpawnSystemAttachedPartModifier "/BRCosmetics/Accessories/FORT_Backpacks/Backpack_M_MED_Clash/FX/NS_Backpack_Clash_Transition.NS_Backpack_Clash_Transition" (CustomCharacter|GetSkeletalMeshForPartOnPlayerPawn) "Root" "0, 0, 0" "0, 0, 0" "SnapToTarget")
    (:Success
      (Variables|Default|SetTransition%20FX _returnvalue_2))
    (:Failure
      _))
  (Utilities|FlowControl|Sequence
    (:then_0
      (bind _keys (Utilities|Map|Keys _ns_stage3_idle_float))
      (bind _values (Utilities|Map|Values _ns_stage3_idle_float))
      (for (_array_element _array_index) _keys
        (Effects|Components|ParticleSystem|SetFloatParameter _returnvalue _array_element (Utilities|Array|Get(acopy) _values _array_index))))
    (:then_1
      (Effects|Components|ParticleSystem|SetIntParameter _returnvalue "BoneFlowSmallSpawn" 0))
    (:then_2
      (bind _keys_2 (Utilities|Map|Keys _ns_stage3_transition_float))
      (bind _values_2 (Utilities|Map|Values _ns_stage3_transition_float))
      (for (_array_element_2 _array_index_2) _keys_2
        (Effects|Components|ParticleSystem|SetFloatParameter (Variables|Default|GetTransition%20FX) _array_element_2 (Utilities|Array|Get(acopy) _values_2 _array_index_2))))
    (:then_3
      (Utilities|IsValid _transition_fx
        (:Is%20Valid
          (Effects|Components|ParticleSystem|SetIntParameter _transition_fx "BoneFlowPerGroup" 5))))))

(fn PlayAudio
  (params (Stage :type Integer))
  (bind _returnvalue (Audio|SpawnSoundAttached (Variables|Default|GetBackpackSound) (CustomCharacter|GetSkeletalMeshForPartOnPlayerPawn)))
  (Audio|Parameter|SetIntegerParameter _returnvalue "Stage" Stage :!class /Script/AudioExtensions.AudioParameterControllerInterface))

(fn ConstructionScript
  (|CallParentFunction))
