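; Part modifier Blueprint for the Clash backpack cosmetic that stages its Niagara effects during gameplay.
; After the post-res-in event it runs Stage 2, waits briefly, then runs Stage 3.
; Each stage activates the idle VFX component and pushes float parameters from name-to-value maps, plus a bone-flow spawn int.
; Stage 3 also spawns a transition system attached to the part mesh's Root and applies its own parameter map.
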(event EventOnPostResIn
  (|CallParentFunction)
  (CallFunction|Stage2)
  (Utilities|FlowControl|Delay 5.2)
  (CallFunction|Stage3))

(event Custom|Stage%202
  (bind _returnvalue (CallFunction|GetIdleVFXComponent))
  (bind _ns_stage2_idle_float (Variables|Default|GetNS_Stage2_Idle_Float))
  (Components|Activation|Activate (CallFunction|GetIdleVFXComponent) true)
  (Utilities|FlowControl|Sequence
    (:then_0
      (bind _keys (Utilities|Map|Keys _ns_stage2_idle_float))
      (bind _values (Utilities|Map|Values _ns_stage2_idle_float))
      (for (_array_element _array_index) _keys
        (Effects|Components|ParticleSystem|SetFloatParameter _returnvalue _array_element (Utilities|Array|Get(acopy) _values _array_index))))
    (:then_1
      (Effects|Components|ParticleSystem|SetIntParameter _returnvalue "BoneFlowSmallSpawn" 0))))

(event Custom|Stage3
  (bind _returnvalue (CallFunction|GetIdleVFXComponent))
  (bind _transition_fx (Variables|Default|GetTransition%20FX))
  (bind _ns_stage3_transition_float (Variables|Default|GetNS_Stage3_Transition_Float))
  (bind _ns_stage3_idle_float (Variables|Default|GetNS_Stage3_Idle_Float))
  (Components|Activation|Activate (CallFunction|GetIdleVFXComponent) true)
  (bind _returnvalue_2 (PartModifier|SpawnSystem|SpawnSystemAttachedPartModifier "/BRCosmetics/Accessories/FORT_Backpacks/Backpack_M_MED_Clash/FX/NS_Backpack_Clash_Transition.NS_Backpack_Clash_Transition" (CustomCharacter|GetSkeletalMeshForPartOnPlayerPawn) "Root" "0, 0, 0" "0, 0, 0" "SnapToTarget")
    (:Success
      (Variables|Default|SetTransition%20FX _returnvalue_2))
    (:Failure
      _))
  (Utilities|FlowControl|Sequence
    (:then_0
      (bind _keys (Utilities|Map|Keys _ns_stage3_idle_float))
      (bind _values (Utilities|Map|Values _ns_stage3_idle_float))
      (for (_array_element _array_index) _keys
        (Effects|Components|ParticleSystem|SetFloatParameter _returnvalue _array_element (Utilities|Array|Get(acopy) _values _array_index))))
    (:then_1
      (Effects|Components|ParticleSystem|SetIntParameter _returnvalue "BoneFlowSmallSpawn" 0))
    (:then_2
      (bind _keys_2 (Utilities|Map|Keys _ns_stage3_transition_float))
      (bind _values_2 (Utilities|Map|Values _ns_stage3_transition_float))
      (for (_array_element_2 _array_index_2) _keys_2
        (Effects|Components|ParticleSystem|SetFloatParameter (Variables|Default|GetTransition%20FX) _array_element_2 (Utilities|Array|Get(acopy) _values_2 _array_index_2))))
    (:then_3
      (Utilities|IsValid _transition_fx
        (:Is%20Valid
          (Effects|Components|ParticleSystem|SetIntParameter _transition_fx "BoneFlowPerGroup" 5))))))

(fn ConstructionScript
  (|CallParentFunction))
