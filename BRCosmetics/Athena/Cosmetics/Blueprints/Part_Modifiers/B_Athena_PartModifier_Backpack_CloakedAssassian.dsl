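; Part modifier for the Cloaked Assassin back bling, a reactive cosmetic that intensifies through four stages, apparently with kills.
; Stat changes feed the idle effect's NumberKills parameter and spawn an attached OnKill Niagara burst for positive values.
; Stage activation sets emissive and world-position-offset material scalars on the part mesh.
; After res-in, frontend display loops through the stages on a timeline as a preview; otherwise a dynamic material starts dimmed.
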(event Custom|Fire_VFX
  (Effects|Components|ParticleSystem|SetFloatParameter (CallFunction|GetIdleVFXComponent) "User.NumberKills" (Math|Conversions|ToFloat(Integer) (Variables|Default|GetStateInt)))
  (PartModifier|SpawnSystem|SpawnSystemAttachedPartModifier "/BRCosmetics/Accessories/FORT_Backpacks/Backpack_F_MED_Cloaked_Assassin/Effects/NS_Backpack_CloakedAssassin_OnKill.NS_Backpack_CloakedAssassin_OnKill" (CustomCharacter|GetSkeletalMeshForPartOnPlayerPawn) "root"))

(event EventOnPostResIn
  (bind _returnvalue (CustomCharacter|GetSkeletalMeshForPartOnPlayerPawn))
  (bind _cube_dyna_mat (Variables|Default|GetCube%20Dyna%20Mat))
  (|CallParentFunction)
  (bind _returnvalue_2 (Rendering|Material|CreateDynamicMaterialInstance (CustomCharacter|GetSkeletalMeshForPartOnPlayerPawn) 0 0 "None" :!class /Script/Engine.PrimitiveComponent))
  (Variables|Default|SetCube%20Dyna%20Mat _returnvalue_2)
  (if (Reactive|Frontend|ShouldDisplayFrontendReactivity 0 "Backpack")
    (Utilities|FlowControl|Delay 0.2)
    ;[1 1 1] Frontend Cycling
    (Utilities|FlowControl|DoOnce false
      (:Completed
        (switch int (Variables|Default|GetStateInt) :!label _label
          (:0
            ;[0.4662 1 0.8486] Set parameters directly on the skel mesh of character parts
            (Rendering|Material|SetScalarParameterValueonMaterials _returnvalue "Emissive_BrightnessMax" 0.1)
            (Rendering|Material|SetScalarParameterValueonMaterials _returnvalue "Emissive_BrightnessMin" 0.01)
            (Rendering|Material|SetScalarParameterValueonMaterials _returnvalue "Emissive mult" 2.0)
            (Rendering|Material|SetScalarParameterValueonMaterials _returnvalue "WPOMultiply mult" 0.0)
            (Effects|Components|ParticleSystem|SetFloatParameter (CallFunction|GetIdleVFXComponent) "User.NumberKills" (Math|Conversions|ToFloat(Integer) (Variables|Default|GetStateInt)))
            (Utilities|FlowControl|Sequence
              (:then_0
                _)))
          (:1
            (Rendering|Material|SetScalarParameterValueonMaterials _returnvalue "Emissive_BrightnessMax" 0.2)
            (Rendering|Material|SetScalarParameterValueonMaterials _returnvalue "Emissive_BrightnessMin" 0.01)
            (Rendering|Material|SetScalarParameterValueonMaterials _returnvalue "Emissive mult" 6.0)
            (Rendering|Material|SetScalarParameterValueonMaterials _returnvalue "WPOMultiply mult" 0.0)
            (CallFunction|FireVFX))
          (:2
            (Rendering|Material|SetScalarParameterValueonMaterials _returnvalue "Emissive_BrightnessMax" 1.5)
            (Rendering|Material|SetScalarParameterValueonMaterials _returnvalue "Emissive_BrightnessMin" 0.05)
            (Rendering|Material|SetScalarParameterValueonMaterials _returnvalue "Emissive mult" 14.0)
            (Rendering|Material|SetScalarParameterValueonMaterials _returnvalue "WPOMultiply mult" 0.25)
            (CallFunction|FireVFX))
          (:3
            (Rendering|Material|SetScalarParameterValueonMaterials _returnvalue "Emissive_BrightnessMax" 5.0)
            (Rendering|Material|SetScalarParameterValueonMaterials _returnvalue "Emissive_BrightnessMin" 2.0)
            (Rendering|Material|SetScalarParameterValueonMaterials _returnvalue "Emissive mult" 2e+01)
            (Rendering|Material|SetScalarParameterValueonMaterials _returnvalue "WPOMultiply mult" 1.25)
            (CallFunction|FireVFX)))
        (if (< (Variables|Default|GetStateInt) 3)
          (Variables|Default|SetStateInt (+ (Variables|Default|GetStateInt) 1))
          (else
            (Variables|Default|SetStateInt 0)))
        (Timeline|Timeline_0 PlayFromStart
          (:Finished
            (goto _label)))))
    (else
      (Utilities|FlowControl|Delay 1.0)
      (Rendering|Material|SetScalarParameterValue _cube_dyna_mat "Emissive_BrightnessMax" 0.0 :!class /Script/Engine.MaterialInstanceDynamic)
      (Rendering|Material|SetScalarParameterValue _cube_dyna_mat "Emissive_BrightnessMin" 0.0 :!class /Script/Engine.MaterialInstanceDynamic)
      (Rendering|Material|SetScalarParameterValue _cube_dyna_mat "Emissive mult" 0.35 :!class /Script/Engine.MaterialInstanceDynamic)
      (Rendering|Material|SetScalarParameterValue _cube_dyna_mat "WPOMultiply" 0.0 :!class /Script/Engine.MaterialInstanceDynamic))))

(event EventOnStatChanged
  (params (State%20Name :type Name :default "None") (Stat%20Value :type Integer))
  (|CallParentFunction State%20Name Stat%20Value)
  (Effects|Components|ParticleSystem|SetFloatParameter (CallFunction|GetIdleVFXComponent) "User.NumberKills" (Math|Conversions|ToFloat(Integer) Stat%20Value))
  (Utilities|FlowControl|Delay 0.2)
  (if (> Stat%20Value 0)
    (PartModifier|SpawnSystem|SpawnSystemAttachedPartModifier "/BRCosmetics/Accessories/FORT_Backpacks/Backpack_F_MED_Cloaked_Assassin/Effects/NS_Backpack_CloakedAssassin_OnKill.NS_Backpack_CloakedAssassin_OnKill" (CustomCharacter|GetSkeletalMeshForPartOnPlayerPawn) "root")))

(event EventOnCosmeticStateActivated
  (params (Stat%20Name :type Name :default "None") (Stat%20Value :type Integer) (STAGE%20Num%20Being%20Activated :type Integer))
  (bind _returnvalue (CustomCharacter|GetSkeletalMeshForPartOnPlayerPawn))
  (|CallParentFunction Stat%20Name Stat%20Value STAGE%20Num%20Being%20Activated)
  ;[1 1 1] Stage Materials set
  (switch int STAGE%20Num%20Being%20Activated
    (:0
      ;[0.4662 1 0.8486] Set parameters directly on the skel mesh of character parts
      (Rendering|Material|SetScalarParameterValueonMaterials _returnvalue "Emissive_BrightnessMax" 0.1)
      (Rendering|Material|SetScalarParameterValueonMaterials _returnvalue "Emissive_BrightnessMin" 0.01)
      (Rendering|Material|SetScalarParameterValueonMaterials _returnvalue "Emissive mult" 2.0)
      (Rendering|Material|SetScalarParameterValueonMaterials _returnvalue "WPOMultiply mult" 0.0))
    (:1
      (Rendering|Material|SetScalarParameterValueonMaterials _returnvalue "Emissive_BrightnessMax" 0.2)
      (Rendering|Material|SetScalarParameterValueonMaterials _returnvalue "Emissive_BrightnessMin" 0.01)
      (Rendering|Material|SetScalarParameterValueonMaterials _returnvalue "Emissive mult" 6.0)
      (Rendering|Material|SetScalarParameterValueonMaterials _returnvalue "WPOMultiply mult" 0.0))
    (:2
      (Rendering|Material|SetScalarParameterValueonMaterials _returnvalue "Emissive_BrightnessMax" 1.5)
      (Rendering|Material|SetScalarParameterValueonMaterials _returnvalue "Emissive_BrightnessMin" 0.05)
      (Rendering|Material|SetScalarParameterValueonMaterials _returnvalue "Emissive mult" 14.0)
      (Rendering|Material|SetScalarParameterValueonMaterials _returnvalue "WPOMultiply mult" 0.25))
    (:3
      (Rendering|Material|SetScalarParameterValueonMaterials _returnvalue "Emissive_BrightnessMax" 3.0)
      (Rendering|Material|SetScalarParameterValueonMaterials _returnvalue "Emissive_BrightnessMin" 0.2)
      (Rendering|Material|SetScalarParameterValueonMaterials _returnvalue "Emissive mult" 2e+01)
      (Rendering|Material|SetScalarParameterValueonMaterials _returnvalue "WPOMultiply mult" 1.25))))

(fn ConstructionScript
  (|CallParentFunction))
