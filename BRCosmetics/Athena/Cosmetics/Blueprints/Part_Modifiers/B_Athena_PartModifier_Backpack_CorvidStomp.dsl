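; Backpack part modifier Blueprint for the CorvidStomp cosmetic back bling, adding a landing effect to the player's backpack.
; When the pawn lands, it calls the parent handler, then spawns the CorvidStomp Niagara system attached to the root of the part's skeletal mesh on the player pawn.
; The construction script only calls its parent.
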(event EventOnPawnLanded
  (|CallParentFunction)
  (PartModifier|SpawnSystem|SpawnSystemAttachedPartModifier "/BRCosmetics/Accessories/FORT_Backpacks/Backpack_M_MED_CorvidStomp/FX/NS_Backpack_CorvidStomp.NS_Backpack_CorvidStomp" (CustomCharacter|GetSkeletalMeshForPartOnPlayerPawn) "root"))

(fn ConstructionScript
  (|CallParentFunction))
