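; Cozy Mechanic backpack cosmetic part modifier that adds kill-reactive emissive glow and Niagara FX.
; On post-res-in it caches the idle VFX component and binds the backpack AnimBP's trigger to frontend reactivity.
; Outside frontend reactivity it watches the player's kill stat, replaying the glow on positive changes but gating the FX with DoOnce.
; Timelines lerp EmissiveBrightness and drive the Niagara ReactivityTime parameter after reinitializing the system.
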(event Custom|ReactivityFE
  (params (TriggerReactivity :type Boolean))
  (bind _niagara_system (Variables|Default|GetNiagara%20System))
  ; Use this instead to sync your timings pls
  (if (Reactive|Frontend|IsCosmeticObjectDisplayedinFrontendContext 0)
    (Utilities|FlowControl|Sequence
      (:then_0
        ;[1 0.06038 0.2724 0.5] Tech Art Reactivity - FrontEnd
        (if TriggerReactivity
          (bind (_direction _brightness _animationcurve) (Timeline|Timeline_1 PlayFromStart)
            (:Update
              (Rendering|Material|SetScalarParameterValueonMaterials (CustomCharacter|GetSkeletalMeshForPartOnPlayerPawn) "EmissiveBrightness" (Math|Float|Lerp (Variables|Default|GetReactivity%20Emissive%20Brightness%20Min) (Variables|Default|GetReactivity%20Emissive%20Brightness%20Max) _animationcurve))))))
      (:then_1
        ;[0.4696 1 0.4929 0.5] FX FE Reactivity
        (if TriggerReactivity
          (Niagara|ReinitializeSystem _niagara_system)
          (Utilities|FlowControl|Sequence
            (:then_0
              (bind (_direction_2 _reactivitytime) (Timeline|FE_Timeline PlayFromStart)
                (:Update
                  (Effects|Components|ParticleSystem|SetFloatParameter _niagara_system "ReactivityTime" _reactivitytime))))
            (:then_1
              (Utilities|FlowControl|DoOnce false
                (:Completed
                  (Effects|Components|ParticleSystem|SetBoolParameter _niagara_system "ShouldFadeFE" false))))))))))

(event EventOnPostResIn
  (bind _niagara_system (Variables|Default|GetNiagara%20System))
  (|CallParentFunction)
  (bind _asskeletal_mesh_component (Utilities|Casting|CastToSkeletalMeshComponent (CustomCharacter|GetSkeletalMeshForPartOnPlayerPawn)))
  (bind _asf_med_cozy_mechanic_pack_anim_bp (Utilities|Casting|CastToF_MED_CozyMechanic_Pack_AnimBP (Components|SkeletalMesh|GetAnimInstance _asskeletal_mesh_component)))
  (Default|BindEventtoCozyMechanicFETriggerFX _asf_med_cozy_mechanic_pack_anim_bp (delegate Custom|ReactivityFE))
  (bind _asniagara_particle_system_component (Utilities|Casting|CastToNiagaraComponent (CallFunction|GetIdleVFXComponent)))
  (Variables|Default|SetNiagara%20System _asniagara_particle_system_component)
  ;[1 1 1] Read CIDs
  (Variables|Default|SetFE_ReactivityEnabled (Reactive|Frontend|IsCosmeticObjectDisplayedinFrontendContext 0))
  (if (not (Reactive|Frontend|ShouldDisplayFrontendReactivity 0 "Backpack"))
    (bind (_statname _statvalue) (Cosmetics|WatchCosmeticStat (CustomCharacter|GetAssociatedPlayerPawn) "(TagName=\"StatManager.AthenaKills\")" 0 false)
      (:OnStatChanged
        (Utilities|FlowControl|Sequence
          (:then_0
            ;[0.3369 0.6856 0.8609 0.5] FX Reactivity
            (if (> _statvalue 0)
              (Utilities|FlowControl|DoOnce false
                (:Completed
                  (Niagara|ReinitializeSystem _niagara_system)
                  (Utilities|FlowControl|Sequence
                    (:then_0
                      (Utilities|FlowControl|Sequence
                        (:then_0
                          (bind (_direction _reactivitytime) (Timeline|FX_Timeline PlayFromStart)
                            (:Update
                              (Effects|Components|ParticleSystem|SetFloatParameter _niagara_system "ReactivityTime" _reactivitytime))))
                        (:then_1
                          (Utilities|FlowControl|Delay 3.43))))
                    (:then_1
                      _))
                  (Utilities|FlowControl|DoOnce false
                    (:Completed
                      (Effects|Components|ParticleSystem|SetBoolParameter _niagara_system "ShouldFadeFE" false)))))))
          (:then_1
            ;[1 0.06038 0.2724 0.5] Tech Art Reactivity
            (if (> _statvalue 0)
              (bind (_direction_2 _brightness _animationcurve) (Timeline|Timeline_0 PlayFromStart)
                (:Update
                  (Rendering|Material|SetScalarParameterValueonMaterials (CustomCharacter|GetSkeletalMeshForPartOnPlayerPawn) "EmissiveBrightness" (Math|Float|Lerp (Variables|Default|GetReactivity%20Emissive%20Brightness%20Min) (Variables|Default|GetReactivity%20Emissive%20Brightness%20Max) _animationcurve)))))))))))

(fn ConstructionScript
  (|CallParentFunction))
