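; Part modifier Blueprint for the Crane Anchor backpack cosmetic that makes its materials and idle VFX react to stat changes, apparently eliminations.
; Each trigger increments an elimination counter, stamps game time into the "ElimationTime" material parameter and passes the count to the idle particle system.
; Any stat change fires the trigger, without checking name or value.
; Frontend reactivity previews it on a delayed repeating timer; otherwise spawn sets the material time negative.
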(event EventOnPostResIn
  (|CallParentFunction)
  (if (Reactive|Frontend|ShouldDisplayFrontendReactivity 0 "Backpack")
    (Utilities|FlowControl|Delay 5.0)
    (CallFunction|TriggerReactivity)
    (Utilities|Time|SetTimerbyEvent (delegate Custom|TriggerReactivity) 15.0 true)
    (else
      (Rendering|Material|SetScalarParameterValueonMaterials (CustomCharacter|GetSkeletalMeshForPartOnPlayerPawn) "ElimationTime" -1e+01))))

(event EventOnStatChanged
  (params (State%20Name :type Name :default "None") (Stat%20Value :type Integer))
  (|CallParentFunction State%20Name Stat%20Value)
  (CallFunction|TriggerReactivity))

(event Custom|TriggerReactivity
  (Math|Integer|IncrementInt (Variables|Default|GetEliminations)
    (:%20%20
      (Rendering|Material|SetScalarParameterValueonMaterials (CustomCharacter|GetSkeletalMeshForPartOnPlayerPawn) "ElimationTime" (Utilities|Time|GetGameTimeinSeconds))
      (Effects|Components|ParticleSystem|SetIntParameter (CallFunction|GetIdleVFXComponent) "Eliminations" (Variables|Default|GetEliminations)))))

(fn ConstructionScript
  (|CallParentFunction))
