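; Backpack cosmetic part modifier for the Crush back bling that plays a fuse, confetti explosion and respawn sequence on its firecracker.
; It runs after post-resurrection init in the front end, and in game whenever a stat changes to a positive value, using retriggerable delays there.
; The idle VFX component acts as the fuse, then switches off as Niagara confetti spawns on the backpack, followed later by a spawn effect.
; Firecracker visibility is toggled by writing game-time timestamps into a dynamic material.
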(event EventOnPostResIn
  (bind _returnvalue (CallFunction|GetIdleVFXComponent))
  (bind _debug (Variables|Default|GetDebug))
  (|CallParentFunction)
  (if (CustomCosmeticModifier|IsinFrontEnd)
    (Components|Activation|Activate _returnvalue false)
    (Utilities|FlowControl|Delay 3.0)
    (Components|Activation|Deactivate _returnvalue)
    (CallFunction|HideFireCracker)
    (PartModifier|SpawnSystem|SpawnSystemAttachedPartModifier "/BRCosmetics/Accessories/FORT_Backpacks/Backpack_F_MED_Crush/FX/NS_Backpack_Crush_Explosion_Confetti.NS_Backpack_Crush_Explosion_Confetti" (CustomCharacter|GetSkeletalMeshForPartOnPlayerPawn) "Ex_FX" "0, 0, 0" "0, 0, 0" "KeepRelativeOffset" true true _debug
      (:Success
        _)
      (:Failure
        _))
    (Utilities|FlowControl|Delay 3.0)
    (CallFunction|RevealFireCracker)
    (PartModifier|SpawnSystem|SpawnSystemAttachedPartModifier "/BRCosmetics/Accessories/FORT_Backpacks/Backpack_F_MED_Crush/FX/NS_Backpack_Crush_Spawn.NS_Backpack_Crush_Spawn" (CustomCharacter|GetSkeletalMeshForPartOnPlayerPawn) "Ex_Fx" "0, 0, 0" "0, 0, 0" "KeepRelativeOffset" true true _debug
      (:Success
        _)
      (:Failure
        _))))

(event EventOnStatChanged
  (params (State%20Name :type Name :default "None") (Stat%20Value :type Integer))
  (bind _returnvalue (CallFunction|GetIdleVFXComponent))
  (bind _debug (Variables|Default|GetDebug))
  (|CallParentFunction State%20Name Stat%20Value)
  (if (> Stat%20Value 0)
    (Components|Activation|Activate _returnvalue false)
    (Utilities|FlowControl|RetriggerableDelay 1.4)
    (Components|Activation|Deactivate _returnvalue)
    (CallFunction|HideFireCracker)
    (PartModifier|SpawnSystem|SpawnSystemAttachedPartModifier "/BRCosmetics/Accessories/FORT_Backpacks/Backpack_F_MED_Crush/FX/NS_Backpack_Crush_Explosion_Confetti.NS_Backpack_Crush_Explosion_Confetti" (CustomCharacter|GetSkeletalMeshForPartOnPlayerPawn) "Ex_FX" "0, 0, 0" "0, 0, 0" "KeepRelativeOffset" true true _debug
      (:Success
        _)
      (:Failure
        _))
    (Utilities|FlowControl|RetriggerableDelay 3.0)
    (CallFunction|RevealFireCracker)
    (PartModifier|SpawnSystem|SpawnSystemAttachedPartModifier "/BRCosmetics/Accessories/FORT_Backpacks/Backpack_F_MED_Crush/FX/NS_Backpack_Crush_Spawn.NS_Backpack_Crush_Spawn" (CustomCharacter|GetSkeletalMeshForPartOnPlayerPawn) "Ex_FX" "0, 0, 0" "0, 0, 0" "KeepRelativeOffset" true true _debug)))

(fn HideFireCracker
  (bind _asmaterial_instance_dynamic (Utilities|Casting|CastToMaterialInstanceDynamic (Rendering|Material|GetMaterial (CustomCharacter|GetSkeletalMeshForPartOnPlayerPawn) 1)))
  (Rendering|Material|SetScalarParameterValue _asmaterial_instance_dynamic "TimestampA" 0.0 :!class /Script/Engine.MaterialInstanceDynamic)
  (Rendering|Material|SetScalarParameterValue _asmaterial_instance_dynamic "TimestampB" (Utilities|Time|GetGameTimeinSeconds) :!class /Script/Engine.MaterialInstanceDynamic))

(fn RevealFireCracker
  (bind _asmaterial_instance_dynamic (Utilities|Casting|CastToMaterialInstanceDynamic (Rendering|Material|GetMaterial (CustomCharacter|GetSkeletalMeshForPartOnPlayerPawn) 1)))
  (Rendering|Material|SetScalarParameterValue _asmaterial_instance_dynamic "TimestampA" (Utilities|Time|GetGameTimeinSeconds) :!class /Script/Engine.MaterialInstanceDynamic)
  (Rendering|Material|SetScalarParameterValue _asmaterial_instance_dynamic "TimestampB" 1e+10 :!class /Script/Engine.MaterialInstanceDynamic))

(fn ConstructionScript
  (|CallParentFunction))
