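; Cosmetic part modifier for the Cube Ninja backpack cape that escalates its runes, glow and VFX through activated stages.
; Cosmetic state changes trigger Stages 1-3, while in the frontend a timeline plays stages 1-4 as a preview.
; Stages 1-3 spawn a transition Niagara system, apply per-stage parameter maps to it, the materials and, from Stage 2, the idle VFX.
; Timelines animate glow and trail; Stage 4 animates them out, then deactivates idle VFX.
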(event EventOnPostResIn
  (|CallParentFunction)
  (if (CustomCosmeticModifier|IsinFrontEnd)
    (Timeline|Frontend PlayFromStart
      (:Stage1
        (CallFunction|Stage1))
      (:Stage2
        (CallFunction|Stage2))
      (:Stage3
        (CallFunction|Stage3))
      (:Stage4
        (CallFunction|Stage4)))))

(event EventOnCosmeticStateActivated
  (params (Stat%20Name :type Name :default "None") (Stat%20Value :type Integer) (STAGE%20Num%20Being%20Activated :type Integer))
  (|CallParentFunction Stat%20Name Stat%20Value STAGE%20Num%20Being%20Activated)
  (switch int STAGE%20Num%20Being%20Activated
    (:1
      (CallFunction|Stage1))
    (:2
      (CallFunction|Stage2))
    (:3
      (CallFunction|Stage3))))

(event Custom|Stage%201
  (bind _returnvalue (CustomCharacter|GetSkeletalMeshForPartOnPlayerPawn))
  (bind _mat_stage1_backpack_float (Variables|Default|GetMat_Stage1_Backpack_Float))
  (bind _ns_stage1_transition_int32 (Variables|Default|GetNS_Stage1_Transition_Int32))
  (bind _ns_stage1_transition_float (Variables|Default|GetNS_Stage1_Transition_Float))
  (Utilities|FlowControl|Delay 0.3)
  (bind _returnvalue_2 (PartModifier|SpawnSystem|SpawnSystemAttachedPartModifier "/BRCosmetics/Accessories/FORT_Capes/Cape_M_MED_Cube_Ninja/FX/NS_Backpack_Cube_Ninja_Male_Transition.NS_Backpack_Cube_Ninja_Male_Transition" (CustomCharacter|GetSkeletalMeshForPartOnPlayerPawn) "dyn_cape02" "0, 0, 0" "0, 0, 0" "SnapToTarget")
    (:Success
      (Variables|Default|SetTransitionFX _returnvalue_2)
      (Utilities|FlowControl|Sequence
        (:then_0
          (bind _keys (Utilities|Map|Keys _ns_stage1_transition_float))
          (bind _values (Utilities|Map|Values _ns_stage1_transition_float))
          (for (_array_element _array_index) _keys
            (Niagara|SetNiagaraVariable(Float) (Variables|Default|GetTransitionFX) _array_element (Utilities|Array|Get(acopy) _values _array_index))))
        (:then_1
          (bind _keys_2 (Utilities|Map|Keys _ns_stage1_transition_int32))
          (bind _values_2 (Utilities|Map|Values _ns_stage1_transition_int32))
          (for (_array_element_2 _array_index_2) _keys_2
            (Niagara|SetNiagaraVariable(Int32) (Variables|Default|GetTransitionFX) _array_element_2 (Utilities|Array|Get(acopy) _values_2 _array_index_2))))
        (:then_2
          _)))
    (:Failure
      _))
  (Utilities|FlowControl|Delay 0.2)
  (bind _keys_3 (Utilities|Map|Keys _mat_stage1_backpack_float))
  (bind _values_3 (Utilities|Map|Values _mat_stage1_backpack_float))
  (for (_array_element_3 _array_index_3) _keys_3
    (Rendering|Material|SetScalarParameterValueonMaterials (CustomCharacter|GetSkeletalMeshForPartOnPlayerPawn) _array_element_3 (Utilities|Array|Get(acopy) _values_3 _array_index_3)))
  (bind (_direction _emissive_brightnessmax _emissive_brightnessmin _runebasecolor_darkenamount) (Timeline|Stage1_Timeline_Runes2 PlayFromStart)
    (:Update
      (Rendering|Material|SetScalarParameterValueonMaterials _returnvalue "CN_RuneBrightness-Max" _emissive_brightnessmax)
      (Rendering|Material|SetScalarParameterValueonMaterials _returnvalue "CN_RuneBrightness-Min" _emissive_brightnessmin)
      (Rendering|Material|SetScalarParameterValueonMaterials _returnvalue "CN_BaseColorOverideAmount" _runebasecolor_darkenamount))))

(event Custom|Stage%202
  (bind _ns_stage2_transition_float (Variables|Default|GetNS_Stage2_Transition_Float))
  (bind _ns_stage2_transition_int32 (Variables|Default|GetNS_Stage2_Transition_Int32))
  (bind _mat_stage2_backpack_float (Variables|Default|GetMat_Stage2_Backpack_Float))
  (bind _ns_stage2_idle_float (Variables|Default|GetNS_Stage2_Idle_Float))
  (Utilities|FlowControl|Delay 0.3)
  (Components|Activation|Activate (CallFunction|GetIdleVFXComponent) false)
  (bind _returnvalue (PartModifier|SpawnSystem|SpawnSystemAttachedPartModifier "/BRCosmetics/Accessories/FORT_Capes/Cape_M_MED_Cube_Ninja/FX/NS_Backpack_Cube_Ninja_Male_Transition.NS_Backpack_Cube_Ninja_Male_Transition" (CustomCharacter|GetSkeletalMeshForPartOnPlayerPawn) "dyn_cape02" "0, 0, 0" "0, 0, 0" "SnapToTarget")
    (:Success
      (Variables|Default|SetTransitionFX _returnvalue)
      (Utilities|FlowControl|Sequence
        (:then_0
          (bind _keys (Utilities|Map|Keys _ns_stage2_transition_int32))
          (bind _values (Utilities|Map|Values _ns_stage2_transition_int32))
          (for (_array_element _array_index) _keys
            (Niagara|SetNiagaraVariable(Int32) (Variables|Default|GetTransitionFX) _array_element (Utilities|Array|Get(acopy) _values _array_index))))
        (:then_1
          (bind _keys_2 (Utilities|Map|Keys _ns_stage2_transition_float))
          (bind _values_2 (Utilities|Map|Values _ns_stage2_transition_float))
          (for (_array_element_2 _array_index_2) _keys_2
            (Niagara|SetNiagaraVariable(Float) (Variables|Default|GetTransitionFX) _array_element_2 (Utilities|Array|Get(acopy) _values_2 _array_index_2))))
        (:then_2
          _)
        (:then_3
          (Utilities|FlowControl|Delay 0.2 :!label _label)
          (bind _keys_3 (Utilities|Map|Keys _mat_stage2_backpack_float))
          (bind _values_3 (Utilities|Map|Values _mat_stage2_backpack_float))
          (for (_array_element_3 _array_index_3) _keys_3
            (Rendering|Material|SetScalarParameterValueonMaterials (CustomCharacter|GetSkeletalMeshForPartOnPlayerPawn) _array_element_3 (Utilities|Array|Get(acopy) _values_3 _array_index_3)))
          (bind (_direction _edgeemimult) (Timeline|Stage2_Timeline_EdgeEmsMult PlayFromStart)
            (:Update
              (Rendering|Material|SetScalarParameterValueonMaterials (CustomCharacter|GetSkeletalMeshForPartOnPlayerPawn) "CN_EdgeEmsMult" _edgeemimult))))))
    (:Failure
      (Utilities|FlowControl|Sequence
        (:then_0
          _)
        (:then_1
          (goto _label)))))
  (bind _keys_4 (Utilities|Map|Keys _ns_stage2_idle_float))
  (bind _values_4 (Utilities|Map|Values _ns_stage2_idle_float))
  (for (_array_element_4 _array_index_4) _keys_4
    (Effects|Components|ParticleSystem|SetFloatParameter (CallFunction|GetIdleVFXComponent) _array_element_4 (Utilities|Array|Get(acopy) _values_4 _array_index_4))))

(event Custom|Stage%203
  (bind _returnvalue (CallFunction|GetIdleVFXComponent))
  (bind _mat_stage3_backpack_float (Variables|Default|GetMat_Stage3_Backpack_Float))
  (bind _ns_stage3_transition_int32 (Variables|Default|GetNS_Stage3_Transition_Int32))
  (bind _ns_stage3_transition_float (Variables|Default|GetNS_Stage3_Transition_Float))
  (bind _ns_stage3_idle_float (Variables|Default|GetNS_Stage3_Idle_Float))
  (Utilities|FlowControl|Delay 0.3)
  (Components|Activation|Activate (CallFunction|GetIdleVFXComponent) false)
  (bind _returnvalue_2 (PartModifier|SpawnSystem|SpawnSystemAttachedPartModifier "/BRCosmetics/Accessories/FORT_Capes/Cape_M_MED_Cube_Ninja/FX/NS_Backpack_Cube_Ninja_Male_Transition.NS_Backpack_Cube_Ninja_Male_Transition" (CustomCharacter|GetSkeletalMeshForPartOnPlayerPawn) "dyn_cape02" "0, 0, 0" "0, 0, 0" "SnapToTarget")
    (:Success
      (Variables|Default|SetTransitionFX _returnvalue_2)
      (Utilities|FlowControl|Sequence
        (:then_0
          (bind _keys (Utilities|Map|Keys _ns_stage3_transition_int32))
          (bind _values (Utilities|Map|Values _ns_stage3_transition_int32))
          (for (_array_element _array_index) _keys
            (Niagara|SetNiagaraVariable(Int32) (Variables|Default|GetTransitionFX) _array_element (Utilities|Array|Get(aref) _values _array_index))))
        (:then_1
          (bind _keys_2 (Utilities|Map|Keys _ns_stage3_transition_float))
          (bind _values_2 (Utilities|Map|Values _ns_stage3_transition_float))
          (for (_array_element_2 _array_index_2) _keys_2
            (Niagara|SetNiagaraVariable(Float) (Variables|Default|GetTransitionFX) _array_element_2 (Utilities|Array|Get(acopy) _values_2 _array_index_2))))
        (:then_2
          _)
        (:then_3
          (Utilities|FlowControl|Delay 0.2 :!label _label)
          (bind _keys_3 (Utilities|Map|Keys _mat_stage3_backpack_float))
          (bind _values_3 (Utilities|Map|Values _mat_stage3_backpack_float))
          (for (_array_element_3 _array_index_3) _keys_3
            (Rendering|Material|SetScalarParameterValueonMaterials (CustomCharacter|GetSkeletalMeshForPartOnPlayerPawn) _array_element_3 (Utilities|Array|Get(acopy) _values_3 _array_index_3)))
          (bind (_direction _cn_edgeemsmult) (Timeline|Stage3_Timeline_EdgeEmsMult PlayFromStart)
            (:Update
              (Rendering|Material|SetScalarParameterValueonMaterials (CustomCharacter|GetSkeletalMeshForPartOnPlayerPawn) "CN_EdgeEmsMult" _cn_edgeemsmult))))))
    (:Failure
      (Utilities|FlowControl|Sequence
        (:then_0
          _)
        (:then_1
          (goto _label)))))
  (Utilities|FlowControl|Delay 0.2)
  (bind _keys_4 (Utilities|Map|Keys _ns_stage3_idle_float))
  (bind _values_4 (Utilities|Map|Values _ns_stage3_idle_float))
  (for (_array_element_4 _array_index_4) _keys_4
    (Effects|Components|ParticleSystem|SetFloatParameter (CallFunction|GetIdleVFXComponent) _array_element_4 (Utilities|Array|Get(acopy) _values_4 _array_index_4)))
  (bind (_direction_2 _trailalpha _light) (Timeline|Stage3_Timeline_Trail PlayFromStart)
    (:Update
      (Effects|Components|ParticleSystem|SetFloatParameter _returnvalue "TrailAlpha" _trailalpha)
      (Effects|Components|ParticleSystem|SetFloatParameter _returnvalue "Light" _light))))

(event Custom|Stage4
  (bind _returnvalue (CustomCharacter|GetSkeletalMeshForPartOnPlayerPawn))
  (bind _returnvalue_2 (CallFunction|GetIdleVFXComponent))
  (bind _returnvalue_3 (CallFunction|GetIdleVFXComponent))
  (Effects|Components|ParticleSystem|SetFloatParameter _returnvalue_3 "SpawnRate" 0.0)
  (Effects|Components|ParticleSystem|SetFloatParameter _returnvalue_3 "Smoke" 0.0)
  (bind (_direction _trail_light _emissive_brightnessmax _emissive_brightnessmin _emissive _emissive_edge _basecoloroverrideamount _rune_basecolor) (Timeline|Stage3_Timeline_Out PlayFromStart)
    (:Update
      (Effects|Components|ParticleSystem|SetFloatParameter _returnvalue_2 "TrailAlpha" _trail_light)
      (Effects|Components|ParticleSystem|SetFloatParameter _returnvalue_2 "Light" _trail_light)
      (Rendering|Material|SetScalarParameterValueonMaterials _returnvalue "CN_RuneBrightness-Max" _emissive_brightnessmax)
      (Rendering|Material|SetScalarParameterValueonMaterials _returnvalue "CN_RuneBrightness-Min" _emissive_brightnessmin)
      (Rendering|Material|SetScalarParameterValueonMaterials _returnvalue "emissive mult" _emissive)
      (Rendering|Material|SetScalarParameterValueonMaterials _returnvalue "CN_EdgeEmsMult" _emissive_edge)
      (Rendering|Material|SetScalarParameterValueonMaterials _returnvalue "CN_BaseColorOverideAmount" _basecoloroverrideamount)
      (Rendering|Material|SetScalarParameterValueonMaterials _returnvalue "CN_RuneBaseColor" _rune_basecolor))
    (:Finished
      (Components|Activation|Deactivate (CallFunction|GetIdleVFXComponent)))))

(fn ConstructionScript
  (|CallParentFunction))
