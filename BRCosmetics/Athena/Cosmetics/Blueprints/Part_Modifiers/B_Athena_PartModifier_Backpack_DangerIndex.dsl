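; Part modifier for the Danger Index backpack cosmetic that fakes audio reactivity on its idle Niagara VFX in the frontend.
; After res-in, it casts the idle VFX component to Niagara and stores it.
; If frontend reactivity should display, timed delays using the fake amplitude off and on lengths switch the EnableFrontendFakeAudioLoop float on, then off.
; A final delay ends the sequence with no visible repeat.
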(event EventOnPostResIn
  (|CallParentFunction)
  (bind _asniagara_particle_system_component (Utilities|Casting|CastToNiagaraComponent (CallFunction|GetIdleVFXComponent)))
  (bind _output_get (Variables|Default|SetNiagara%20System _asniagara_particle_system_component))
  (if (Reactive|Frontend|ShouldDisplayFrontendReactivity 0 "Backpack")
    (Utilities|FlowControl|Delay (Variables|AudioReactivity|GetAudioReactivityFakeAmplitudeOffLength))
    (Niagara|SetNiagaraVariable(Float) _output_get "EnableFrontendFakeAudioLoop" 1.0)
    (Utilities|FlowControl|Delay (Variables|AudioReactivity|GetAudioReactivityFakeAmplitudeOnLength))
    (Niagara|SetNiagaraVariable(Float) _output_get "EnableFrontendFakeAudioLoop" 0.0)
    (Utilities|FlowControl|Delay 5.0)))

(fn ConstructionScript
  (|CallParentFunction))
