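; Part modifier Blueprint for the Dark Eagle Fire backpack cosmetic that toggles its idle VFX around skydiving.
; The begin-skydiving event hides the component returned by GetIdleVFXComponent, and the end-skydiving event shows it again.
; Visibility changes use Set Hidden in Game without propagating to child components, and the construction script adds no logic.
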(event EventOnBeginSkydiving
  (Development|SetHiddeninGame (CallFunction|GetIdleVFXComponent) true false))

(event EventOnEndSkydiving
  (Development|SetHiddeninGame (CallFunction|GetIdleVFXComponent) false false))

(fn ConstructionScript)
