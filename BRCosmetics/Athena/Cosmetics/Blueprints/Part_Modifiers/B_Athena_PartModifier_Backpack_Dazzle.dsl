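; Part modifier for the Dazzle back bling that, when reactive, bursts and rumbles in response to player kills.
; At setup it reads the Progressive variant, then in match watches the kill stat, storing stages up to three.
; Each reaction spawns an attached Niagara burst fed variant and stage, while a timeline sine jitters mesh scale.
; The frontend previews this after a delay with its own cycling stage counter, without sound.
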(event EventOnPostResIn
  ;[1 0.2176 0.2574 1 200] EVENT 🖥️
  (|CallParentFunction)
  ;[0.2545 0.2899 0.1934 1 200] SETUP⚙
  (if (Variables|Default|GetIs%20Reactive)
    (bind _asfort_player_pawn (Utilities|Casting|CastToFortPlayerPawn (Actor|GetOwner)))
    (bind _index (CosmeticLoadout|GetVariantChannelActiveIndexforSlotActor "(TagName=\"LoadoutSlot.Backpack\")" _asfort_player_pawn "(TagName=\"Cosmetics.Variant.Channel.Progressive\")" (Utilities|Array|MakeArray "(TagName=\"Cosmetics.Variant.Property.Stage1\")" "(TagName=\"Cosmetics.Variant.Property.Stage2\")")))
    (switch int _index
      (:0
        (Variables|Default|SetbVariant? false))
      (:1
        (Variables|Default|SetbVariant? true)))
    (if (Reactive|Frontend|ShouldDisplayFrontendReactivity 0 "Backpack")
      ;[0.141 0.4403 1 1 200] FRONTEND 🛒
      (Utilities|FlowControl|DoOnce false
        (:Completed
          (Utilities|FlowControl|Delay 2.0)
          (Variables|Default|SetFE_KillStage (+ (Math|Integer|%25%28Integer%29 (Variables|Default|GetFE_KillStage) 3) 1))
          ;;[0 0 0] Trigger "OnKill" Audio
          (CallFunction|TriggerAudio)
          (bind _returnvalue (PartModifier|SpawnSystem|SpawnSystemAttachedPartModifier "/BRCosmetics/Accessories/FORT_Backpacks/Backpack_F_MED_Dazzle/FX/NS_Backpack_Dazzle_Burst02.NS_Backpack_Dazzle_Burst02" (CustomCharacter|GetSkeletalMeshForPartOnPlayerPawn) "Root" "0, 0, 0" "0, 0, 0" "SnapToTarget")
            (:Success
              (if (Variables|Default|GetbVariant?)
                (Niagara|SetNiagaraVariable(Float) _returnvalue "VariantID" 1.0)
                (else
                  (Niagara|SetNiagaraVariable(Float) _returnvalue "VariantID" 0.0)))
              (Effects|Components|ParticleSystem|SetIntParameter _returnvalue "KillStage" (Variables|Default|GetFE_KillStage))
              ;;[0.0625 0.1161 0.2083] rumble
              (bind (_direction _newtrack_2) (Timeline|Timeline_0 PlayFromStart)
                (:Update
                  (Transformation|SetWorldScale3D (CustomCharacter|GetSkeletalMeshForPartOnPlayerPawn) (Math|Conversions|ToVector(Float) (Utilities|RemapFloat (Math|Conversions|ToFloat(Integer) (Math|Float|Floor (Math|Trig|Sin(Radians) _newtrack_2))) "(X=-1.000000,Y=1.000000)" "(X=0.950000,Y=1.000000)"))))
                (:Finished
                  (Utilities|FlowControl|Delay 2.5)))))))
      (else
        ;[1 0.7262 0.2176 1 200] ON KILL REACTIVITY ☠
        (Utilities|FlowControl|DoOnce false
          (:Completed
            (bind (_statname _statvalue) (Cosmetics|WatchCosmeticStat (CustomCharacter|GetAssociatedPlayerPawn) "(TagName=\"StatManager.AthenaKills\")")
              (:OnStatChanged
                (if (<= _statvalue 3)
                  (bind _output_get (Variables|Default|SetKillStage _statvalue)))
                (if (> (Math|Conversions|ToFloat(Integer) _output_get) 0.0)
                  ;;[0 0 0] Trigger "OnKill" Audio
                  (CallFunction|TriggerAudio)
                  (bind _returnvalue_2 (PartModifier|SpawnSystem|SpawnSystemAttachedPartModifier "/BRCosmetics/Accessories/FORT_Backpacks/Backpack_F_MED_Dazzle/FX/NS_Backpack_Dazzle_Burst02.NS_Backpack_Dazzle_Burst02" (CustomCharacter|GetSkeletalMeshForPartOnPlayerPawn) "Root" "0, 0, 0" "0, 0, 0" "SnapToTarget")
                    (:Success
                      (if (Variables|Default|GetbVariant?)
                        (Niagara|SetNiagaraVariable(Float) _returnvalue_2 "VariantID" 1.0)
                        (else
                          (Niagara|SetNiagaraVariable(Float) _returnvalue_2 "VariantID" 0.0)))
                      (Effects|Components|ParticleSystem|SetIntParameter _returnvalue_2 "KillStage" (Variables|Default|GetKillStage))
                      (bind (_direction_2 _newtrack_2_2) (Timeline|Timeline PlayFromStart)
                        (:Update
                          (Transformation|SetWorldScale3D (CustomCharacter|GetSkeletalMeshForPartOnPlayerPawn) (Math|Conversions|ToVector(Float) (Utilities|RemapFloat (Math|Conversions|ToFloat(Integer) (Math|Float|Floor (Math|Trig|Sin(Radians) _newtrack_2_2))) "(X=-1.000000,Y=1.000000)" "(X=0.950000,Y=1.000000)"))))))))))))))))

(event Custom|TriggerAudio
  (bind _sound (Variables|Default|GetSound))
  ;[0 0 0 1 200] Audio🔊
  (if (Class|ItemUIFunctionLibrary|IsNullItem _sound)
    _
    (elif (Reactive|Frontend|ShouldDisplayFrontendReactivity 0 "Backpack")
      _
      (else
        (Audio|SpawnSoundAttached _sound (CustomCharacter|GetSkeletalMeshForPartOnPlayerPawn))))))

(fn ConstructionScript
  (|CallParentFunction))
