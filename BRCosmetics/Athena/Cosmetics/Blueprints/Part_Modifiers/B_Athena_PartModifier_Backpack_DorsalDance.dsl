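; Cosmetic part modifier for the DorsalDance backpack that animates a water level in its materials through the WaterSurfaceHeight scalar parameter.
; On post-init it sets the bottom level; in the front end it also sets emissive intensity and plays the WaterPreview timeline.
; In game, Tick adds height from pawn speed, interpolates it back toward the bottom, and pushes the clamped result via WaterMaterialUpdate.
; Time and pawn forward/right vectors are stored but not read here.
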(event EventOnPostResIn
  (|CallParentFunction)
  (Rendering|Material|SetScalarParameterValueonMaterials (CustomCharacter|GetSkeletalMeshForPartOnPlayerPawn) "WaterSurfaceHeight" (Variables|Default|GetWaterBottomHeight))
  (if (CustomCosmeticModifier|IsinFrontEnd)
    (Rendering|Material|SetScalarParameterValueonMaterials (CustomCharacter|GetSkeletalMeshForPartOnPlayerPawn) "EmissiveIntensityOverall" 0.7)
    (bind (_direction _waterheightalpha) (Timeline|WaterPreview Play)
      (:Update
        (Rendering|Material|SetScalarParameterValueonMaterials (CustomCharacter|GetSkeletalMeshForPartOnPlayerPawn) "WaterSurfaceHeight" (+ (* (Math|Float|Absolute(Float) (Variables|Default|GetWaterBottomHeight)) _waterheightalpha) (Variables|Default|GetWaterBottomHeight)))))))

(event EventTick
  (params (DeltaSeconds :type Float%20%28single-precision%29))
  (bind _returnvalue (CustomCharacter|GetAssociatedPlayerPawn))
  (bind _charactermovement (Variables|/Script/Engine.Character|GetCharacterMovement _returnvalue))
  (bind _returnvalue_2 (Pawn|Components|CharacterMovement|GetLastUpdateRotation _charactermovement))
  (if (not (CustomCosmeticModifier|IsinFrontEnd))
    (Variables|Default|SetTime (+ DeltaSeconds (Variables|Default|GetTime)))
    (Utilities|FlowControl|Sequence
      (:then_0
        (Variables|Default|SetHieghtAmountToAdd (Math|Interpolation|FInterpTo (Variables|Default|GetHieghtAmountToAdd) (Variables|Default|GetWaterBottomHeight) DeltaSeconds (Variables|Default|GetHeightRecoverySpeed))))
      (:then_1
        (Variables|Default|SetWaterHeight (Math|Float|Clamp(Float) (Variables|Default|GetHieghtAmountToAdd) (Variables|Default|GetWaterBottomHeight) (Variables|Default|GetMaxWaterHeight)))
        (CallFunction|WaterMaterialUpdate))
      (:then_2
        (Variables|Default|SetVelocity (/ (Transformation|GetVelocity _returnvalue) (Variables|Default|GetSpeedDivisor)))
        (Variables|Default|SetCharacterPawnForwardVector (Math|Vector|GetForwardVector _returnvalue_2))
        (Variables|Default|SetCharacterPawnRightVector (Math|Vector|GetRightVector _returnvalue_2)))
      (:then_3
        (Variables|Default|SetHieghtAmountToAdd (+ (Variables|Default|GetHieghtAmountToAdd) (Math|Float|Clamp(Float) (* (* (Math|Float|Clamp(Float) (Math|Vector|VectorLength (Variables|Default|GetVelocity))) 0.02) (* (Variables|Default|GetWaterBottomHeight) -1.0)) 0.0 (* (Variables|Default|GetWaterBottomHeight) -1.0))))))))

(event Custom|WaterMaterialUpdate
  ;[0 0 0] Material Set up
  (Rendering|Material|SetScalarParameterValueonMaterials (CustomCharacter|GetSkeletalMeshForPartOnPlayerPawn) "WaterSurfaceHeight" (Variables|Default|GetWaterHeight)))

(event EventOnDayPhaseChanged
  (params (CurrentDayPhase :type EFortDayPhase%20Enum :default "Morning") (PreviousDayPhase :type EFortDayPhase%20Enum :default "Morning") (bAtCreation :type Boolean)))

(event EventOnNewPlayerWeaponType
  (params (NewType :type EFortItemType%20Enum :default "WorldItem") (OldType :type EFortItemType%20Enum :default "WorldItem"))
  (|CallParentFunction NewType OldType) ; This node is disabled and will not be called.\nDrag off pins to build functionality.
)

(event EventOnCurieStateAttached
  (params (Tag :type Gameplay%20Tag%20Structure%20%28by%20ref%29)))

(event EventOnWeaponChanged
  (params (NewWeapon :type Fort%20Weapon%20Object%20Reference) (OldWeapon :type Fort%20Weapon%20Object%20Reference)))

(event EventOnCurieStateDetached
  (params (Tag :type Gameplay%20Tag%20Structure%20%28by%20ref%29)))

(event EventOnWeaponFired)

(event EventOnBeginSkydiving
  (|CallParentFunction) ; This node is disabled and will not be called.\nDrag off pins to build functionality.
)

(event EventOnNewWeapons
  (params (WeaponsAsItems :type Array%20of%20Fort%20Item%20Object%20References)))

(event EventOnEndSkydiving
  (|CallParentFunction) ; This node is disabled and will not be called.\nDrag off pins to build functionality.
)

(event EventOnEnterVehicle
  (params (VehicleEntered :type Fort%20Vehicle%20Interface%20Interface%20%28by%20ref%29))
  (|CallParentFunction VehicleEntered) ; This node is disabled and will not be called.\nDrag off pins to build functionality.
)

(event EventOnExitVehicle
  (|CallParentFunction) ; This node is disabled and will not be called.\nDrag off pins to build functionality.
)

(event EventOnTargetingOrFirstPersonCameraChanged)

(event EventOnHitPawn
  (params (SelfActor :type Actor%20Object%20Reference) (OtherActor :type Actor%20Object%20Reference) (NormalImpulse :type Vector :default "0, 0, 0") (Hit :type Hit%20Result%20Structure%20%28by%20ref%29)))

(fn ConstructionScript
  (|CallParentFunction))
