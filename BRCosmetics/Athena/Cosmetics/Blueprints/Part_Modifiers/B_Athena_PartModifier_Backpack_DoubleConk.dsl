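; Cosmetic part modifier for the DoubleConk backpack that plays a screen-switch and glitch effect when the player takes damage or is downed.
; On res-in it caches the idle Niagara component, then watches the DamageReceived stat, or in the frontend plays a separate preview timeline after a delay.
; Timelines drive mesh material scalars, plus Niagara variables for backpack parts.
; Retriggers during later playback rewind to a fixed point instead of restarting.
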(event EventOnDayPhaseChanged
  (params (CurrentDayPhase :type EFortDayPhase%20Enum :default "Morning") (PreviousDayPhase :type EFortDayPhase%20Enum :default "Morning") (bAtCreation :type Boolean)))

(event EventOnNewPlayerWeaponType
  (params (NewType :type EFortItemType%20Enum :default "WorldItem") (OldType :type EFortItemType%20Enum :default "WorldItem"))
  (|CallParentFunction NewType OldType) ; This node is disabled and will not be called.\nDrag off pins to build functionality.
)

(event EventOnWeaponChanged
  (params (NewWeapon :type Fort%20Weapon%20Object%20Reference) (OldWeapon :type Fort%20Weapon%20Object%20Reference))
  (|CallParentFunction NewWeapon OldWeapon) ; This node is disabled and will not be called.\nDrag off pins to build functionality.
)

(event EventOnWeaponFired)

(event EventOnNewWeapons
  (params (WeaponsAsItems :type Array%20of%20Fort%20Item%20Object%20References)))

(event EventOnCurieStateAttached
  (params (Tag :type Gameplay%20Tag%20Structure%20%28by%20ref%29)))

(event EventOnPostResIn
  (bind _returnvalue (CustomCharacter|GetSkeletalMeshForPartOnPlayerPawn))
  (bind _niagara_system (Variables|Default|GetNiagara%20System))
  (|CallParentFunction)
  (bind _asniagara_particle_system_component (Utilities|Casting|CastToNiagaraComponent (CallFunction|GetIdleVFXComponent)))
  (Variables|Default|SetNiagara%20System _asniagara_particle_system_component)
  (if (CustomCosmeticModifier|IsinFrontEnd)
    (Utilities|FlowControl|Delay 4.0)
    (Components|Timeline|SetPlaybackPosition (Variables|B_Athena_PartModifier_Backpack_DoubleConk|GetTL%20Reactivity%20FE) 7.0 false true)
    (bind (_direction _reactivity _glitchspawn) (Timeline|TL%20Reactivity%20FE Play)
      (:Update
        (Rendering|Material|SetScalarParameterValueonMaterials _returnvalue "EmissiveBrightness" (Math|Float|Lerp 1.0 0.0 _reactivity))
        (Rendering|Material|SetScalarParameterValueonMaterials _returnvalue "DCBP_ScreenSwitchState" _reactivity)
        (switch Utilities|FlowControl|Switch|SwitchonEFortCustomPartType (CustomCharacter|GetCharacterPartType)
          (:Backpack
            (Niagara|SetNiagaraVariable(Float) _niagara_system "Reactivity" _reactivity)
            (Niagara|SetNiagaraVariable(Float) _niagara_system "GlitchSpawn" _glitchspawn)))))
    (else
      (Cosmetics|WatchCosmeticStat (CustomCharacter|GetAssociatedPlayerPawn) "(TagName=\"StatManager.DamageReceived\")" 0 false
        (:OnStatChanged
          (CallFunction|TriggerReactivityFX))))))

(event Custom|Trigger%20Reactivity%20FX
  (bind _returnvalue (CustomCharacter|GetSkeletalMeshForPartOnPlayerPawn))
  (bind _niagara_system (Variables|Default|GetNiagara%20System))
  (if (and (Components|Timeline|IsPlaying (Variables|B_Athena_PartModifier_Backpack_DoubleConk|GetTL%20Reactivity)) (>= (Components|Timeline|GetPlaybackPosition (Variables|B_Athena_PartModifier_Backpack_DoubleConk|GetTL%20Reactivity)) 0.75))
    (CallFunction|RetriggerReactivity)
    (else
      (bind (_direction _reactivity _glitchspawn) (Timeline|TL%20Reactivity PlayFromStart)
        (:Update
          (Rendering|Material|SetScalarParameterValueonMaterials _returnvalue "EmissiveBrightness" (Math|Float|Lerp 1.0 0.0 _reactivity))
          (Rendering|Material|SetScalarParameterValueonMaterials _returnvalue "DCBP_ScreenSwitchState" _reactivity)
          (switch Utilities|FlowControl|Switch|SwitchonEFortCustomPartType (CustomCharacter|GetCharacterPartType)
            (:Backpack
              (Niagara|SetNiagaraVariable(Float) _niagara_system "Reactivity" _reactivity)
              (Niagara|SetNiagaraVariable(Float) _niagara_system "GlitchSpawn" _glitchspawn))))))))

(event EventOnCurieStateDetached
  (params (Tag :type Gameplay%20Tag%20Structure%20%28by%20ref%29)))

(event EventOnBeginSkydiving
  (|CallParentFunction) ; This node is disabled and will not be called.\nDrag off pins to build functionality.
)

(event EventPlayerStartDBNO
  (|CallParentFunction)
  (CallFunction|TriggerReactivityFX))

(event EventOnEndSkydiving
  (|CallParentFunction) ; This node is disabled and will not be called.\nDrag off pins to build functionality.
)

(event EventOnEnterVehicle
  (params (VehicleEntered :type Fort%20Vehicle%20Interface%20Interface%20%28by%20ref%29))
  (|CallParentFunction VehicleEntered) ; This node is disabled and will not be called.\nDrag off pins to build functionality.
)

(event EventOnExitVehicle
  (|CallParentFunction) ; This node is disabled and will not be called.\nDrag off pins to build functionality.
)

(event Custom|Retrigger%20Reactivity
  (Components|Timeline|SetPlaybackPosition (Variables|B_Athena_PartModifier_Backpack_DoubleConk|GetTL%20Reactivity) 0.75 false true))

(event EventOnTargetingOrFirstPersonCameraChanged)

(event EventOnHitPawn
  (params (SelfActor :type Actor%20Object%20Reference) (OtherActor :type Actor%20Object%20Reference) (NormalImpulse :type Vector :default "0, 0, 0") (Hit :type Hit%20Result%20Structure%20%28by%20ref%29)))

(fn ConstructionScript
  (|CallParentFunction))
