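; Backpack part modifier for the DriftSummer cosmetic that switches the idle VFX component on and off in response to player movement events.
; The effect deactivates when skydiving begins or a vehicle is entered, and reactivates when skydiving ends, a vehicle is exited or the pawn jumps.
; Each handler calls its parent first, and the end of skydiving also resets the particle SpawnRate parameter to a fixed value.
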(event EventOnBeginSkydiving
  (|CallParentFunction)
  (Components|Activation|Deactivate (CallFunction|GetIdleVFXComponent)))

(event EventOnEndSkydiving
  (bind _returnvalue (CallFunction|GetIdleVFXComponent))
  (|CallParentFunction)
  (Components|Activation|Activate _returnvalue true)
  (Effects|Components|ParticleSystem|SetFloatParameter _returnvalue "SpawnRate" 8e+01))

(event EventOnEnterVehicle
  (params (VehicleEntered :type Fort%20Vehicle%20Interface%20Interface%20%28by%20ref%29))
  (|CallParentFunction VehicleEntered)
  (Components|Activation|Deactivate (CallFunction|GetIdleVFXComponent)))

(event EventOnExitVehicle
  (|CallParentFunction)
  (Components|Activation|Activate (CallFunction|GetIdleVFXComponent) true))

(event EventOnPawnJumped
  (|CallParentFunction)
  (Components|Activation|Activate (CallFunction|GetIdleVFXComponent) true))

(fn ConstructionScript)
