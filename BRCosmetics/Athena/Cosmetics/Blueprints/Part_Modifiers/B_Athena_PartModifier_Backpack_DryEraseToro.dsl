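; Part modifier for the DryEraseToro back bling that adds elimination reactivity when the equipped Parts variant is Stage1.
; In game it triggers OnKill when the AthenaKills stat changes; in frontend previews it loops OnKill on a timer.
; OnKill toggles Elims material parameters and, if the mesh is visible, spawns timed Niagara glow and prop effects.
; Staged sound-library cues accompany the effects outside the frontend.
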(event EventOnPostResIn
  (|CallParentFunction)
  (Utilities|FlowControl|Sequence
    (:then_0
      ;[0.5 0.5 0.5 0.15] Read CIDs
      (bind (_index _bindexisvalid) (Class|FortCosmeticFunctionLibrary|GetVariantChannelActiveIndexForItem "/BRCosmetics/Athena/Items/Cosmetics/Backpacks/Backpack_DryEraseToro.Backpack_DryEraseToro" (CosmeticLoadout|GetCosmeticLoadout) "(TagName=\"Cosmetics.Variant.Channel.Parts\")" (Utilities|Array|MakeArray "(TagName=\"Cosmetics.Variant.Property.Stage1\")") false false))
      (Variables|Default|SetReactivityEnabled _bindexisvalid))
    (:then_1
      (if (Variables|Default|GetReactivityEnabled)
        (if (Reactive|Frontend|ShouldDisplayFrontendReactivity 0 "Backpack")
          ; Frontend Loop
          (Utilities|FlowControl|Delay 4.0)
          (CallFunction|OnKill)
          (Utilities|Time|SetTimerbyEvent (delegate Custom|OnKill) 15.0 true)
          (else
            ; Setup On Kill
            (Cosmetics|WatchCosmeticStat (CustomCharacter|GetAssociatedPlayerPawn) "(TagName=\"StatManager.AthenaKills\")" 0 false
              (:OnStatChanged
                (CallFunction|OnKill)))))))))

(event Custom|OnKill
  (bind _returnvalue (CustomCharacter|GetSkeletalMeshForPartOnPlayerPawn))
  (bind _returnvalue_2 (CustomCharacter|GetSkeletalMeshForPartOnPlayerPawn))
  ; On Kill Reactivity
  ;;[1 0 0.09835 0.5] Material
  (Rendering|Material|SetScalarParameterValueonMaterials _returnvalue "Elims" 1.0)
  (Rendering|Material|SetScalarParameterValueonMaterials _returnvalue "Elims_end" 0.0)
  (if (Rendering|IsVisible (CustomCharacter|GetSkeletalMeshForPartOnPlayerPawn))
    (Niagara|SpawnSystemAttached "/BRCosmetics/Accessories/FORT_Backpacks/Backpack_M_MED_DryEraseToro/FX/NS_Backpack_DryEraseToro_GlowLights.NS_Backpack_DryEraseToro_GlowLights" (CustomCharacter|GetSkeletalMeshForPartOnPlayerPawn) "pack" "0, 0, 0" "0, 0, 0" "KeepRelativeOffset" true)
    (Utilities|FlowControl|Delay 3.5)
    (bind _returnvalue_3 (Niagara|SpawnSystemAttached "/BRCosmetics/Accessories/FORT_Backpacks/Backpack_M_MED_DryEraseToro/FX/NS_Backpack_DryEraseToro_SpawnProp.NS_Backpack_DryEraseToro_SpawnProp" (CustomCharacter|GetSkeletalMeshForPartOnPlayerPawn) "pack" "0, 0, 0" "0, 0, 0" "KeepRelativeOffset" true))
    (Effects|Components|ParticleSystem|SetBoolParameter _returnvalue_3 "SpawnShelfFX" true)
    (Effects|Components|ParticleSystem|SetBoolParameter _returnvalue_3 "SpawnDoorFX" true)
    (Effects|Components|ParticleSystem|SetBoolParameter _returnvalue_3 "SpawnConeLightFX" true)
    (CallFunction|PlayReactiveSound 0)
    (Utilities|FlowControl|Delay 4.7)
    ;;[1 0 0.09835 0.5] Material
    (Rendering|Material|SetScalarParameterValueonMaterials _returnvalue_2 "Elims" 0.0)
    (Rendering|Material|SetScalarParameterValueonMaterials _returnvalue_2 "Elims_end" 1.0)
    (Rendering|Material|SetScalarParameterValueonMaterials _returnvalue_2 "Time Start" (Utilities|Time|GetGameTimeinSeconds))
    (bind _returnvalue_4 (Niagara|SpawnSystemAttached "/BRCosmetics/Accessories/FORT_Backpacks/Backpack_M_MED_DryEraseToro/FX/NS_Backpack_DryEraseToro_SpawnProp.NS_Backpack_DryEraseToro_SpawnProp" (CustomCharacter|GetSkeletalMeshForPartOnPlayerPawn) "pack" "0, 0, 0" "0, 0, 0" "KeepRelativeOffset" true))
    (Effects|Components|ParticleSystem|SetBoolParameter _returnvalue_4 "SpawnShelfFX" true)
    (Effects|Components|ParticleSystem|SetBoolParameter _returnvalue_4 "SpawnDoorFX" true)
    (CallFunction|PlayReactiveSound 1)))

(event Custom|PlayReactiveSound
  (params (Index :type Integer))
  ; Custom reactive audio - should not play in FrontEnd
  (if (not (Reactive|Frontend|ShouldDisplayFrontendReactivity 0 "Backpack"))
    (SoundLibrary|PlaySoundSimple (WorldSubsystems|GetSoundLibrarySubsystem) (CustomCharacter|GetAssociatedPlayerPawn) (Utilities|Struct|MakeSimpleSoundSettings (Utilities|Select :Index Index :Option%200 "(TagName=\"SoundLibrary.Cosmetic.Other.One\")" :Option%201 "(TagName=\"SoundLibrary.Cosmetic.Other.Two\")") true false true false 1.0 1.0 0.0 0 0 "None" (CustomCharacter|GetSkeletalMeshForPartOnPlayerPawn)) :!class /Script/SoundLibrary.SoundLibrarySubsystem)))

(fn SetAndPlaySound
  (Audio|SpawnSoundAttached "/BRCosmetics/Sounds/Backblings/Apprentice/BackBling_Apprentice_Elimination_Open_Cue.BackBling_Apprentice_Elimination_Open_Cue" (CustomCharacter|GetSkeletalMeshForPartOnPlayerPawn)))

(fn ConstructionScript
  (|CallParentFunction))
