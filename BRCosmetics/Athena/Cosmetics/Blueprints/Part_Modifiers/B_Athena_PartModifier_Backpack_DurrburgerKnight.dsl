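; Backpack cosmetic part modifier for the DurrburgerKnight back bling that fires a particle burst when the pawn lands.
; On landing, it calls the parent handler, then sets the JumpBurst float parameter on the idle VFX particle component to one.
; After a short delay it resets that parameter to zero, turning the burst into a brief pulse.
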(event EventOnPawnLanded
  (bind _returnvalue (CallFunction|GetIdleVFXComponent))
  (|CallParentFunction)
  (Effects|Components|ParticleSystem|SetFloatParameter _returnvalue "JumpBurst" 1.0)
  (Utilities|FlowControl|Delay 0.1)
  (Effects|Components|ParticleSystem|SetFloatParameter _returnvalue "JumpBurst" 0.0))

(fn ConstructionScript
  (|CallParentFunction))
