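; Backpack part modifier for the Dyed Duelist cosmetic that plays an emissive boost in reaction to kills.
; When reactive, it watches the player's kill stat and restarts a timeline on change.
; The timeline drives the OnKillVibrantEmissiveAmount material parameter on the part's skeletal mesh.
; In frontend reactivity preview, a looping timer replays the same timeline and is cleared when preview should not display.
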(event Custom|FrontEnd_ReactivityPreview
  ;[1 1 1] Displays effect on frontend
  (if (Reactive|Frontend|ShouldDisplayFrontendReactivity 0 "Backpack")
    (bind (_direction _boost) (Timeline|Timeline_1 PlayFromStart)
      (:Update
        (Rendering|Material|SetScalarParameterValueonMaterials (CustomCharacter|GetSkeletalMeshForPartOnPlayerPawn) "OnKillVibrantEmissiveAmount" _boost)))
    (else
      (Utilities|Time|ClearandInvalidateTimerbyHandle (Variables|Default|GetFrontEndReactivityTimer)))))

(event EventOnPawnAndPartChanged
  (|CallParentFunction)
  (Utilities|FlowControl|Sequence
    (:then_0
      (if (Variables|Default|GetIs%20Reactive)
        ;[1 1 1] On Kill Event
        (Cosmetics|WatchCosmeticStat (CustomCharacter|GetAssociatedPlayerPawn) "(TagName=\"StatManager.AthenaKills\")" 0 false
          (:OnStatChanged
            (Timeline|Timeline_1 PlayFromStart)))))
    (:then_1
      ;[1 0.1636 0.04388] Frontend reactivity preview setup
      (if (Reactive|Frontend|ShouldDisplayFrontendReactivity 0 "Backpack")
        (bind _returnvalue (Utilities|Time|SetTimerbyFunctionName 0 "FrontEnd_ReactivityPreview" 7.2 true))
        (Variables|Default|SetFrontEndReactivityTimer _returnvalue)
        (else
          (Utilities|Time|ClearandInvalidateTimerbyHandle (Variables|Default|GetFrontEndReactivityTimer)))))))

(fn ConstructionScript)
