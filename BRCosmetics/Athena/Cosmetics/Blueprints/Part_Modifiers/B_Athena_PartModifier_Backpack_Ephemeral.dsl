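; Cosmetic part modifier for the Ephemeral backpack that changes its glow based on whether the player holds the harvesting tool.
; On weapon change it waits a tick, then plays an emissive timeline from the start for a harvest weapon, or reverses it otherwise.
; The timeline drives EmissiveBrightness on the part's skeletal mesh materials.
; A PlayAudio function spawns the backpack sound attached to that mesh outside the front end, with its Stage parameter set.
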(event EventOnWeaponChanged
  (params (NewWeapon :type Fort%20Weapon%20Object%20Reference) (OldWeapon :type Fort%20Weapon%20Object%20Reference))
  (|CallParentFunction NewWeapon OldWeapon)
  (Utilities|FlowControl|DelayUntilNextTick)
  (if (Utilities|Enum|Equal(Enum) (Item|GetWeaponDataItemType (Variables|/Script/FortniteGame.FortPawn|GetCurrentWeapon (CustomCharacter|GetAssociatedPlayerPawn))) "WeaponHarvest")
    (bind (_direction _emissiveness) (Timeline|Emissive%20Timeline PlayFromStart)
      (:Update
        (Rendering|Material|SetScalarParameterValueonMaterials (CustomCharacter|GetSkeletalMeshForPartOnPlayerPawn) "EmissiveBrightness" _emissiveness)))
    (else
      (Timeline|Emissive%20Timeline Reverse))))

(event EventOnPawnAndPartChanged
  (|CallParentFunction))

(fn PlayAudio
  (if (not (CustomCosmeticModifier|IsinFrontEnd))
    (bind _returnvalue (Audio|SpawnSoundAttached (Variables|Default|GetBackpackSound) (CustomCharacter|GetSkeletalMeshForPartOnPlayerPawn)))
    (Audio|Parameter|SetIntegerParameter _returnvalue "Stage" 0 :!class /Script/AudioExtensions.AudioParameterControllerInterface)))

(fn ConstructionScript)
