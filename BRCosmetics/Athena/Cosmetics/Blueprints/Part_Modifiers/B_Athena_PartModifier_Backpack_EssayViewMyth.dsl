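; Cosmetic part modifier for the EssayViewMyth back bling that fires a reactive effect on eliminations in a match, or on an animation event in the frontend.
; After res-in it binds to the pack AnimBP's FETriggerFX in the frontend; otherwise it resets ReactivityTime and watches the AthenaKills stat.
; ReactivityFX writes the current game time to ReactivityTime on the idle VFX and mesh materials.
; Both branches set IsInFrontend to zero.
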(event Custom|ReactivityFX
  (Effects|Components|ParticleSystem|SetFloatParameter (CallFunction|GetIdleVFXComponent) "ReactivityTime" (Utilities|Time|GetGameTimeinSeconds))
  (Rendering|Material|SetScalarParameterValueonMaterials (Variables|Default|GetSkeletalMesh) "ReactivityTime" (Utilities|Time|GetGameTimeinSeconds)))

(event EventOnPostResIn
  (bind _skeletalmesh (Variables|Default|GetSkeletalMesh))
  (|CallParentFunction)
  (Variables|Default|SetSkeletalMesh (CustomCharacter|GetSkeletalMeshForPartOnPlayerPawn))
  (if (Reactive|Frontend|IsCosmeticObjectDisplayedinFrontendContext 0)
    (Rendering|Material|SetScalarParameterValueonMaterials (Variables|Default|GetSkeletalMesh) "IsInFrontend" 0.0)
    (bind _asskeletal_mesh_component (Utilities|Casting|CastToSkeletalMeshComponent (Variables|Default|GetSkeletalMesh)))
    (bind _asm_med_essay_view_myth_pack_anim_bp (Utilities|Casting|CastToM_MED_EssayViewMyth_Pack_AnimBP (Components|SkeletalMesh|GetAnimInstance _asskeletal_mesh_component)))
    (Default|BindEventtoFETriggerFX _asm_med_essay_view_myth_pack_anim_bp (delegate Custom|ReactivityFX))
    (else
      (Rendering|Material|SetScalarParameterValueonMaterials _skeletalmesh "IsInFrontend" 0.0)
      (Rendering|Material|SetScalarParameterValueonMaterials _skeletalmesh "ReactivityTime" -1e+01)
      (bind (_statname _statvalue) (Cosmetics|WatchCosmeticStat (CustomCharacter|GetAssociatedPlayerPawn) "(TagName=\"StatManager.AthenaKills\")" -1 false)
        (:OnStatChanged
          (if (> _statvalue 0)
            (CallFunction|ReactivityFX)))))))

(fn ConstructionScript
  (|CallParentFunction))
