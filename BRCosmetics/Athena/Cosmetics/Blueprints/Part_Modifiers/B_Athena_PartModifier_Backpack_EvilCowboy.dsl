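; Part modifier for the Evil Cowboy backpack, a reactive cosmetic with two stages driven by cosmetic state activation.
; Each stage sets material colors and scalars from per-stage maps and spawns an attached Niagara transition with loadout variants; stage 2 also reveals idle VFX.
; Post res-in starts an elimination stat watch and, for frontend previews, plays both stages in a delayed one-shot sequence.
; Elim Burst replays the first transition effect but has no caller here.
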(event Custom|Stage%201
  (bind _returnvalue (CustomCharacter|GetAssociatedPlayerPawn))
  (bind _returnvalue_2 (CustomCharacter|GetCharacterPartType))
  (bind _returnvalue_3 (Customization|GetSkeletalMeshforPartType _returnvalue _returnvalue_2 :!class /Script/FortniteGame.FortPlayerPawn))
  (bind _stage_1_material_colors (Variables|Default|GetStage%201%20Material%20Colors))
  (bind _stage_1_material_scalars (Variables|Default|GetStage%201%20Material%20Scalars))
  (Networking|CanExecuteCosmeticEvents
    (:True
      (Utilities|FlowControl|Sequence
        (:then_0
          ;[1 1 1] Stage 1 Material Changes for Character Part
          (bind _keys (Utilities|Map|Keys _stage_1_material_colors))
          (if (CallFunction|ArrayIsValid _keys)
            (for _array_element _keys
              (if (Utilities|Name|Equal(Name) _array_element "None")
                (Development|PrintString "What are you thinking?  A Variable called None.  Really?!")
                (else
                  (Rendering|Material|SetVectorParameterValueonMaterials _returnvalue_3 _array_element (Math|Conversions|ToVector(LinearColor) (Utilities|Map|Find _stage_1_material_colors _array_element))))))
            (bind _keys_2 (Utilities|Map|Keys _stage_1_material_scalars))
            (if (CallFunction|ArrayIsValid _keys_2)
              (for _array_element_2 _keys_2
                (if (Utilities|Name|Equal(Name) _array_element_2 "None")
                  (Development|PrintString "What are you thinking?  A Variable called None.  Really?!")
                  (else
                    (Rendering|Material|SetScalarParameterValueonMaterials _returnvalue_3 _array_element_2 (Utilities|Map|Find _stage_1_material_scalars _array_element_2))))))))
        (:then_1
          ;[1 1 1] Stage 1 VFX
          (bind _returnvalue_4 (Niagara|SpawnSystemAttached "/BRCosmetics/Accessories/FORT_Backpacks/M_MED_Evil_Cowboy_Backpack/FX/NS_BackPack_EvilCowboy_Stage2_Transition.NS_BackPack_EvilCowboy_Stage2_Transition" (CustomCharacter|GetSkeletalMeshForPartOnPlayerPawn) "None" "0, 0, 0" "0, 0, 0" "SnapToTarget" true))
          (Niagara|SetNiagaraVariable(Bool) _returnvalue_4 "bIsFrontEndPreview" (CallFunction|ShouldDisplayFrontendReactivity))
          (Cosmetics|ApplyVariantsToComponent(Deprecated) :self (Variables|Default|GetBackpack%20Item%20Definition%20Reference) :InComponent _returnvalue_4 :Loadout (Variables|/Script/FortniteGame.FortPlayerPawn|GetCosmeticLoadout (CustomCharacter|GetAssociatedPlayerPawn)))
          (Components|Activation|Activate _returnvalue_4 false))))))

(event Custom|Elim%20Burst
  ;[1 1 1] Stage 1 VFX
  (bind _returnvalue (Niagara|SpawnSystemAttached "/BRCosmetics/Accessories/FORT_Backpacks/M_MED_Evil_Cowboy_Backpack/FX/NS_BackPack_EvilCowboy_Stage2_Transition.NS_BackPack_EvilCowboy_Stage2_Transition" (CustomCharacter|GetSkeletalMeshForPartOnPlayerPawn) "None" "0, 0, 0" "0, 0, 0" "SnapToTarget" true))
  (Niagara|SetNiagaraVariable(Bool) _returnvalue "bIsFrontEndPreview" (CallFunction|ShouldDisplayFrontendReactivity))
  (Cosmetics|ApplyVariantsToComponent(Deprecated) :self (Variables|Default|GetBackpack%20Item%20Definition%20Reference) :InComponent _returnvalue :Loadout (Variables|/Script/FortniteGame.FortPlayerPawn|GetCosmeticLoadout (CustomCharacter|GetAssociatedPlayerPawn)))
  (Components|Activation|Activate _returnvalue false))

(event EventOnCosmeticStateActivated
  (params (Stat%20Name :type Name :default "None") (Stat%20Value :type Integer) (STAGE%20Num%20Being%20Activated :type Integer))
  (|CallParentFunction Stat%20Name Stat%20Value STAGE%20Num%20Being%20Activated)
  (switch int STAGE%20Num%20Being%20Activated
    (:Default
      (Development|PrintString (Utilities|String|Append "Unhandled Cosmetic Stage " (Utilities|String|ToString(Integer) STAGE%20Num%20Being%20Activated)) true true "(R=1.000000,G=0.000000,B=0.006820,A=1.000000)"))
    (:1
      (CallFunction|Stage1))
    (:2
      (CallFunction|Stage2))))

(event Custom|Stage%202
  (bind _returnvalue (CustomCharacter|GetAssociatedPlayerPawn))
  (bind _returnvalue_2 (CustomCharacter|GetCharacterPartType))
  (bind _returnvalue_3 (Customization|GetSkeletalMeshforPartType _returnvalue _returnvalue_2 :!class /Script/FortniteGame.FortPlayerPawn))
  (bind _stage_2_material_colors (Variables|Default|GetStage%202%20Material%20Colors))
  (bind _stage_2_material_scalars (Variables|Default|GetStage%202%20Material%20Scalars))
  (Networking|CanExecuteCosmeticEvents
    (:True
      (Utilities|FlowControl|Sequence
        (:then_0
          ;[1 1 1] Stage 1 Material Changes for Character Part
          (bind _keys (Utilities|Map|Keys _stage_2_material_colors))
          (if (CallFunction|ArrayIsValid _keys)
            (for _array_element _keys
              (if (Utilities|Name|Equal(Name) _array_element "None")
                (Development|PrintString "What are you thinking?  A Variable called None.  Really?!")
                (else
                  (Rendering|Material|SetVectorParameterValueonMaterials _returnvalue_3 _array_element (Math|Conversions|ToVector(LinearColor) (Utilities|Map|Find _stage_2_material_colors _array_element))))))
            (bind _keys_2 (Utilities|Map|Keys _stage_2_material_scalars))
            (if (CallFunction|ArrayIsValid _keys_2)
              (for _array_element_2 _keys_2
                (if (Utilities|Name|Equal(Name) _array_element_2 "None")
                  (Development|PrintString "What are you thinking?  A Variable called None.  Really?!")
                  (else
                    (Rendering|Material|SetScalarParameterValueonMaterials _returnvalue_3 _array_element_2 (Utilities|Map|Find _stage_2_material_scalars _array_element_2))))))))
        (:then_1
          ;[1 1 1] Stage 2 VFX
          (bind _returnvalue_4 (Niagara|SpawnSystemAttached "/BRCosmetics/Accessories/FORT_Backpacks/M_MED_Evil_Cowboy_Backpack/FX/NS_BackPack_EvilCowboy_Stage3_Transition.NS_BackPack_EvilCowboy_Stage3_Transition" (CustomCharacter|GetSkeletalMeshForPartOnPlayerPawn) "None" "0, 0, 0" "0, 0, 0" "SnapToTarget" true))
          (Niagara|SetNiagaraVariable(Bool) _returnvalue_4 "bIsFrontEndPreview" (CallFunction|ShouldDisplayFrontendReactivity))
          (Cosmetics|ApplyVariantsToComponent(Deprecated) :self (Variables|Default|GetBackpack%20Item%20Definition%20Reference) :InComponent _returnvalue_4 :Loadout (Variables|/Script/FortniteGame.FortPlayerPawn|GetCosmeticLoadout (CustomCharacter|GetAssociatedPlayerPawn)))
          (Components|Activation|Activate _returnvalue_4 false)
          (Effects|Components|ParticleSystem|SetBoolParameter (CallFunction|GetIdleVFXComponent) "bIsVisible" true))))))

(event EventOnPostResIn
  ; Adding On Elim Tracking, uses burst stage 1
  (|CallParentFunction)
  (bind (_statname _statvalue) (Cosmetics|WatchCosmeticStat (CustomCharacter|GetAssociatedPlayerPawn) "(TagName=\"StatManager.AthenaKills\")" 0 false -1.0 0.1))
  (CallFunction|FrontEndLoop))

(event Custom|FrontEndLoop
  (if (CallFunction|ShouldDisplayFrontendReactivity)
    (Utilities|FlowControl|Delay 2.0)
    (CallFunction|Stage1)
    (Utilities|FlowControl|Delay 3.0)
    (CallFunction|Stage2)))

(fn ArrayIsValid
  (params (Name%20Array :type Array%20of%20Names))
  (return (> (Utilities|Array|Length Name%20Array) 0)))

(fn ConstructionScript
  (|CallParentFunction))
