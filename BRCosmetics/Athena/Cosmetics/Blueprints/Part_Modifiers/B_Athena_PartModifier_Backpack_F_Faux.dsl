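; Backpack part modifier for the Faux cosmetic that tints the backpack's skeletal mesh through its BackpackTestColor material vector parameter.
; ToFemale and ToMale anim notify events set the color to blue or red and store state 0 or 1 respectively.
; OnPostResIn calls the parent, then switches on that stored state to set the color again.
; The parameter name suggests a test visual.
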(event EventOnPostResIn
  (bind _returnvalue (CustomCharacter|GetSkeletalMeshForPartOnPlayerPawn))
  (|CallParentFunction)
  (switch int (Variables|Default|GetCurrentState)
    (:Default
      _)
    (:0
      (Rendering|Material|SetVectorParameterValueonMaterials _returnvalue "BackpackTestColor" "0, 0,1.000000"))
    (:1
      _))
  (Rendering|Material|SetVectorParameterValueonMaterials _returnvalue "BackpackTestColor" "1.000000,0.000000,0.000000"))

(event Custom|AnimNotify_Faux_ToFemale
  (Rendering|Material|SetVectorParameterValueonMaterials (CustomCharacter|GetSkeletalMeshForPartOnPlayerPawn) "BackpackTestColor" "0, 0,1.000000")
  (Variables|Default|SetCurrentState 0))

(event Custom|AnimNotify_Faux_ToMale
  (Rendering|Material|SetVectorParameterValueonMaterials (CustomCharacter|GetSkeletalMeshForPartOnPlayerPawn) "BackpackTestColor" "1.000000,0.000000,0.000000")
  (Variables|Default|SetCurrentState 1))

(fn ConstructionScript)
