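; Cosmetic part modifier for the FeatherMudGlance back bling, animating its material in reaction to player movement and kills.
; A looping timer polls pawn velocity, playing a run timeline when movement starts and resetting when it stops.
; Watched kill-stat changes activate idle VFX and a glow timeline that resets after a delay, while stat thresholds set the current stage.
; Frontend display disables tick and plays a preview timeline instead.
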(event EventOnPawnAndPartChanged
  ; Stage Set
  (|CallParentFunction)
  (if (not (CustomCosmeticModifier|IsinFrontEnd))
    (bind _returnvalue (Class|FortCosmeticFunctionLibrary|GetCosmeticStatValue (CustomCharacter|GetAssociatedPlayerPawn) "(TagName=\"StatManager.AthenaKills\")"))
    (bind _stage (CallFunction|StatValuetoStage _returnvalue))
    (Variables|Default|SetCurrentStage _stage)))

(event EventOnPostResIn
  (bind _returnvalue (CustomCharacter|GetSkeletalMeshForPartOnPlayerPawn))
  (bind _returnvalue_2 (CustomCharacter|GetSkeletalMeshForPartOnPlayerPawn))
  ; Event Timeline
  (|CallParentFunction)
  (if (Reactive|Frontend|IsCosmeticObjectDisplayedinFrontendContext 0)
    (Actor|Tick|SetActorTickEnabled false)
    (Utilities|FlowControl|Sequence
      (:then_0
        (bind (_direction _kill _reset) (Timeline|Timeline_2 PlayFromStart)
          (:Update
            (Rendering|Material|SetScalarParameterValueonMaterials (CustomCharacter|GetSkeletalMeshForPartOnPlayerPawn) "Animation Phase2" _kill)
            (Rendering|Material|SetScalarParameterValueonMaterials (CustomCharacter|GetSkeletalMeshForPartOnPlayerPawn) "Reset radius" _reset))
          (:VFX
            (Components|Activation|Activate (CallFunction|GetIdleVFXComponent) true)
            (CallFunction|Stage1))))
      (:then_2
        (Rendering|Material|SetScalarParameterValueonMaterials (CustomCharacter|GetSkeletalMeshForPartOnPlayerPawn) "Specular multiple" 0.1)))
    (else
      (Utilities|FlowControl|Sequence
        (:then_0
          ; Get player pawn
          (bind _asfort_player_pawn_athena (Utilities|Casting|CastToFortPlayerPawnAthena (CustomCharacter|GetAssociatedPlayerPawn)))
          (Variables|Default|SetAs%20Fort%20Player%20Pawn%20Athena _asfort_player_pawn_athena)
          (Utilities|Time|SetTimerbyEvent (delegate Custom|In%20motion) 0.05 true))
        (:then_1
          ;[0.1418 0.05353 0.15 0.5] When killing an enemy is played animationt
          (Cosmetics|WatchCosmeticStat (CustomCharacter|GetAssociatedPlayerPawn) "(TagName=\"StatManager.AthenaKills\")" 0 false
            (:OnStatChanged
              (Components|Activation|Activate (CallFunction|GetIdleVFXComponent) true)
              (bind (_direction_2 _kill_2 _glow) (Timeline|Timeline_1 PlayFromStart)
                (:Update
                  (Rendering|Material|SetScalarParameterValueonMaterials _returnvalue "Animation Phase2" _kill_2)
                  (Rendering|Material|SetScalarParameterValueonMaterials _returnvalue "KillGlow" _glow))
                (:Finished
                  (Rendering|Material|SetScalarParameterValueonMaterials _returnvalue_2 "Reset radius" 0.0)
                  (Rendering|Material|SetScalarParameterValueonMaterials _returnvalue_2 "Animation Phase2" 0.87)
                  (Utilities|FlowControl|Delay 5.0)
                  (Rendering|Material|SetScalarParameterValueonMaterials _returnvalue_2 "Animation Phase2" 0.0)
                  (bind (_direction_3 _kill_3 _reset_2) (Timeline|Timeline PlayFromStart)
                    (:Update
                      (Rendering|Material|SetScalarParameterValueonMaterials (CustomCharacter|GetSkeletalMeshForPartOnPlayerPawn) "Reset radius" _reset_2)
                      (Rendering|Material|SetScalarParameterValueonMaterials (CustomCharacter|GetSkeletalMeshForPartOnPlayerPawn) "KillGlow" _kill_3))))))))))))

(event Custom|Stage%201
  ; Stages
  (Effects|Components|ParticleSystem|SetBoolParameter (CallFunction|GetIdleVFXComponent) "Stage_01" true))

(event Custom|In%20motion
  (bind _returnvalue (CustomCharacter|GetSkeletalMeshForPartOnPlayerPawn))
  ;[0 0.7396 0.5186 0.5] When moving, play the animation
  (if (CustomCosmeticModifier|IsinFrontEnd)
    _
    (elif (Variables|Default|GetIs%20Playing%20animation)
      _
      (elif (> (Math|Vector|VectorLength (Transformation|GetVelocity (Variables|Default|GetAs%20Fort%20Player%20Pawn%20Athena))) 1.0)
        (if (not (Variables|Default|GetIs%20Run))
          (Variables|Default|SetIs%20Playing%20animation true)
          (bind (_direction _run) (Timeline|Timeline_0 PlayFromStart)
            (:Update
              (Rendering|Material|SetScalarParameterValueonMaterials _returnvalue "Animation Phase" _run))
            (:Finished
              (Variables|Default|SetIs%20Playing%20animation false)
              (Variables|Default|SetIs%20Run true)
              (Rendering|Material|SetScalarParameterValueonMaterials _returnvalue "Animation Phase" 0.62))))
        (else
          (Variables|Default|SetIs%20Run false)
          (Rendering|Material|SetScalarParameterValueonMaterials _returnvalue "Animation Phase" 0.0))))))

(event EventOnCosmeticStateActivated
  (params (Stat%20Name :type Name :default "None") (Stat%20Value :type Integer) (STAGE%20Num%20Being%20Activated :type Integer))
  ; Update Mats
  (|CallParentFunction Stat%20Name Stat%20Value STAGE%20Num%20Being%20Activated)
  (Variables|Default|SetCurrentStage STAGE%20Num%20Being%20Activated))

(fn StatValueToStage
  (params (Stat%20Value :type Integer))
  (bind (_arrayindex _arrayelement) (Utilities|Array|ReverseforEachLoop (Variables|PartModifierControls|WatchCosmeticStat|GetStat%20Thresholds))
    (:LoopBody
      (if (>= (Variables|GetStat%20Value) _arrayelement)
        (return _arrayindex)))
    (:Completed
      (return 0))))

(fn ConstructionScript
  (|CallParentFunction))
