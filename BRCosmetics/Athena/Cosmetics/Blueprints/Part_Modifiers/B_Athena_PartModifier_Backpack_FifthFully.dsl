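; Backpack cosmetic part modifier for FifthFully that fades and scales in a set of wings on the player after respawn-in.
; Construction hides the wings and zeroes their four material opacities.
; OnPostResIn attaches them to the backpack mesh with leader pose, then a timeline drives per-material opacity and world scale.
; Material override events, bound through a frontend-context check, toggle wings and idle VFX for hologram overrides, with leftover debug prints.
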(event EventOnPostResIn
  (bind _returnvalue (CustomCharacter|GetAssociatedPlayerPawn))
  (bind _wings (Variables|Default|GetWings))
  (bind _wings_2 (Variables|Default|GetWings))
  (|CallParentFunction)
  (Utilities|FlowControl|Sequence
    (:then_0
      (bind _asfort_player_pawn (Utilities|Casting|CastToFortPlayerPawn (Actor|GetOwner)))
      (Transformation|AttachComponentToComponent _wings_2 (Customization|GetSkeletalMeshforPartType _asfort_player_pawn "Backpack" :!class /Script/FortniteGame.FortPlayerPawn) "None" "SnapToTarget" "KeepRelative" "KeepRelative" true)
      (Components|SkinnedMesh|SetLeaderPoseComponent _wings_2 (Customization|GetSkeletalMeshforPartType _asfort_player_pawn "Backpack" :!class /Script/FortniteGame.FortPlayerPawn) true false)
      (Rendering|SetVisibility (Variables|Default|GetWings) true false)
      (bind (_direction _opacity _scale) (Timeline|Timeline PlayFromStart)
        (:Update
          (for _index (range 4)
            (bind _returnvalue_2 (Fortnite|GetDynamicMaterialInstance _wings _index))
            (Rendering|Material|SetScalarParameterValue _returnvalue_2 (Utilities|Select :Index _index :Option%200 "Opacity" :Option%201 "Opacity" :Option%202 "Opacity" :Option%203 "OpacityMaskMultiplier") (* _opacity (Utilities|Select :Index _index :Option%200 0.25 :Option%201 0.15 :Option%202 0.075 :Option%203 0.45)) :!class /Script/Engine.MaterialInstanceDynamic))
          (Transformation|SetWorldScale3D _wings (Math|Conversions|ToVector(Float) (Math|Float|Lerp 0.02 1.0 _scale))))))
    (:then_1
      (if (Reactive|Frontend|IsCosmeticObjectDisplayedinFrontendContext 0)
        (EventDispatchers|AssignOnMaterialOverrideApplied _returnvalue (delegate Custom|OnMaterialOverrideApplied_Event))
        (EventDispatchers|AssignOnMaterialOverrideCleared _returnvalue (delegate Custom|OnMaterialOverrideCleared_Event))))))

(event EventTick
  (params (DeltaSeconds :type Float%20%28single-precision%29)))

(event Custom|OnMaterialOverrideApplied_Event
  (params (AffectedPawn :type Fort%20Player%20Pawn%20Object%20Reference) (AppliedOverride_OverrideId :type Guid%20Structure) (AppliedOverride_Material :type Material%20Interface%20Soft%20Object%20Reference))
  ;[0 0 0] hologram
  (Development|PrintString "HologramMaterialApplied")
  (Default|SetHologramOverrides true))

(event Custom|OnMaterialOverrideCleared_Event
  (params (AffectedPawn :type Fort%20Player%20Pawn%20Object%20Reference))
  ;[0 0 0] hologramClear
  (Development|PrintString "HologramMaterialRemoved")
  (Default|SetHologramOverrides false))

(fn SetHologramOverrides
  (params (isHologram :type Boolean))
  (LocalVariables|SetisHolo isHologram)
  (Development|SetHiddeninGame (CallFunction|GetIdleVFXComponent) (LocalVariables|GetisHolo) false)
  (Rendering|SetVisibility (Variables|Default|GetWings) (not (LocalVariables|GetisHolo)) false))

(fn ConstructionScript
  (|CallParentFunction)
  (for _index (range 4)
    (bind _returnvalue (Fortnite|GetDynamicMaterialInstance (Variables|Default|GetWings) _index))
    (Rendering|Material|SetScalarParameterValue _returnvalue (Utilities|Select :Index _index :Option%200 "Opacity" :Option%201 "Opacity" :Option%202 "Opacity" :Option%203 "OpacityMaskMultiplier") 0.0 :!class /Script/Engine.MaterialInstanceDynamic))
  (Rendering|SetVisibility (Variables|Default|GetWings) false false))
