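; Part modifier for the Flapjack Wrangler backpack cosmetic that drives its VFX as the wearing pawn jumps and lands.
; Jump and land events toggle an IsOnGround boolean parameter on the idle VFX component.
; After a landing and a brief delay, it spawns a landing Niagara system attached to the part's skeletal mesh on the player pawn.
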(event EventOnPawnJumped
  (Effects|Components|ParticleSystem|SetBoolParameter (CallFunction|GetIdleVFXComponent) "IsOnGround" false))

(event EventOnPawnLanded
  (Effects|Components|ParticleSystem|SetBoolParameter (CallFunction|GetIdleVFXComponent) "IsOnGround" true)
  (Utilities|FlowControl|Delay 0.125)
  (PartModifier|SpawnSystem|SpawnSystemAttachedPartModifier "/BRCosmetics/Accessories/FORT_Backpacks/Backpack_M_MED_Flapjack_Wrangler/FX/NS_Backpack_Flapjack_Wrangler_Male_Land.NS_Backpack_Flapjack_Wrangler_Male_Land" (CustomCharacter|GetSkeletalMeshForPartOnPlayerPawn)))

(fn ConstructionScript
  (|CallParentFunction))
