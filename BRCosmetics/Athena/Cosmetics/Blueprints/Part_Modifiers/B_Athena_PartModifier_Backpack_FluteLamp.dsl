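; Cosmetic part modifier for the FluteLamp back bling that reacts to the player's eliminations.
; On an elimination, if the part's skeletal mesh is visible, it spawns an attached Niagara transition effect at the "pack" socket and passes it the elim count.
; Post res-in calls the parent, then resets the frontend elim count to zero.
; SetAndPlaySound plays an Apprentice back bling elimination cue on the mesh; nothing here calls it.
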(event EventOnPostResIn
  (|CallParentFunction)
  (Utilities|FlowControl|Sequence
    (:then_0
      (Variables|Default|SetFrontendElimCount 0))))

(event Custom|OnElim
  (params (ElimCount :type Integer))
  ; On Kill Reactivity
  (if (Rendering|IsVisible (CustomCharacter|GetSkeletalMeshForPartOnPlayerPawn))
    (bind _returnvalue (Niagara|SpawnSystemAttached "/BRCosmetics/Accessories/FORT_Backpacks/Backpack_M_MED_FluteLamp/FX/NS_Backpack_FluteLamp_Transition.NS_Backpack_FluteLamp_Transition" (CustomCharacter|GetSkeletalMeshForPartOnPlayerPawn) "pack" "0, 0, 0" "0, 0, 0" "KeepRelativeOffset" true))
    (Effects|Components|ParticleSystem|SetIntParameter _returnvalue "ElimCount" ElimCount)))

(fn SetAndPlaySound
  (Audio|SpawnSoundAttached "/BRCosmetics/Sounds/Backblings/Apprentice/BackBling_Apprentice_Elimination_Open_Cue.BackBling_Apprentice_Elimination_Open_Cue" (CustomCharacter|GetSkeletalMeshForPartOnPlayerPawn)))

(fn ConstructionScript
  (|CallParentFunction))
