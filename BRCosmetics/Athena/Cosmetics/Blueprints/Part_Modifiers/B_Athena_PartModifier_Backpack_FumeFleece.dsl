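; Backpack part modifier for the FumeFleece cosmetic that makes the back bling react to player kills.
; After resource init, it passes a variant index, derived from mesh-name substrings by a placeholder function, to the idle Niagara effect.
; Frontend previews replay the reaction on a looping timer; otherwise a kill-stat watcher triggers it with a cooldown.
; Each reaction activates the effect, sets KillCount and stamps game time into the material.
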(event EventOnPostResIn
  (|CallParentFunction)
  (Utilities|FlowControl|Sequence
    (:then_0
      (bind _asniagara_particle_system_component (Utilities|Casting|CastToNiagaraComponent (CallFunction|GetIdleVFXComponent)))
      (Niagara|SetNiagaraVariable(Int32) _asniagara_particle_system_component "VariantIndex" (CallFunction|FXGetVariantIndex)))
    (:then_1
      (if (Reactive|Frontend|ShouldDisplayFrontendReactivity 0 "Backpack")
        (Utilities|FlowControl|Delay 3.0)
        (CallFunction|OnFrontEndReactivity)
        (Utilities|Time|SetTimerbyEvent (EventDispatchers|CreateEvent :!event Custom|OnFrontEndReactivity) 7.0 true)
        (else
          (bind (_statname _statvalue) (Cosmetics|WatchCosmeticStat (CustomCharacter|GetAssociatedPlayerPawn) "(TagName=\"StatManager.AthenaKills\")" 0 false)
            (:OnStatChanged
              (if (not (Variables|Default|GetIsPlayingReactivity))
                (CallFunction|OnKill _statvalue)
                (Variables|Default|SetIsPlayingReactivity true)
                (Utilities|FlowControl|Delay 5.0)
                (Variables|Default|SetIsPlayingReactivity false)))))))))

(event Custom|OnFrontEndReactivity
  (CallFunction|OnKill -1))

(fn FX_GetVariantIndex
  ; this is placeholder
  (bind _asskeletal_mesh_component (Utilities|Casting|CastToSkeletalMeshComponent (CustomCharacter|GetSkeletalMeshForPartOnPlayerPawn)))
  (LocalVariables|Setobj_name (Utilities|GetObjectName (Variables|/Script/Engine.SkinnedMeshComponent|GetSkeletalMesh _asskeletal_mesh_component)))
  (bind (_array_element _array_index) (Utilities|Array|ForEachLoopwithBreak (LocalVariables|GetSubstrings))
    (:LoopBody
      (if (!= (Utilities|String|FindSubstring (LocalVariables|Getobj_name) _array_element) -1)
        (LocalVariables|Setfound_index _array_index)))
    (:Completed
      _))
  (return (LocalVariables|Getfound_index)))

(fn OnKill
  (params (Count :type Integer))
  (bind _returnvalue (CallFunction|GetIdleVFXComponent))
  (Components|Activation|Activate _returnvalue true)
  (Effects|Components|ParticleSystem|SetIntParameter _returnvalue "KillCount" Count)
  (bind _asmaterial_instance_dynamic (Utilities|Casting|CastToMaterialInstanceDynamic (Rendering|Material|GetMaterial (CustomCharacter|GetSkeletalMeshForPartOnPlayerPawn) 0)))
  (Rendering|Material|SetScalarParameterValue _asmaterial_instance_dynamic "Reactivity_Timestamp" (Utilities|Time|GetGameTimeinSeconds) :!class /Script/Engine.MaterialInstanceDynamic))

(fn ConstructionScript
  (|CallParentFunction))
