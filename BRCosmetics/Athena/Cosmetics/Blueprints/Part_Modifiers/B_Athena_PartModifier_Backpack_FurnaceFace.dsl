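; Backpack part modifier for the FurnaceFace cosmetic that mounts a glow component onto the player's backpack.
; When the res-in post event fires, it calls the parent, then attaches the sm_furnace_face_backpack_glow component to the part's skeletal mesh on the player pawn.
; The attach snaps location, rotation and scale to the Backpack_glow_FX_Socket socket and welds bodies.
; The construction script is empty.
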(event EventOnPostResIn
  (|CallParentFunction)
  (Transformation|AttachComponentToComponent (Variables|Default|Getsm_furnace_face_backpack_glow) (CustomCharacter|GetSkeletalMeshForPartOnPlayerPawn) "Backpack_glow_FX_Socket" "SnapToTarget" "SnapToTarget" "SnapToTarget" true))

(fn ConstructionScript)
