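; Part modifier for the Future Samurai Summer backpack cosmetic that attaches an extra skeletal mesh and applies its red eye variant.
; After post-res-in it waits a tick, attaches its mesh to the part mesh on the player pawn and follows that pose via leader pose.
; When the backpack's Particle variant channel resolves to Particle2, it swaps five material slots to red eye materials.
; All other events are empty or have disabled parent calls.
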(event EventOnPostResIn
  (bind _returnvalue (CustomCharacter|GetSkeletalMeshForPartOnPlayerPawn))
  (bind _skeletalmesh (Variables|Default|GetSkeletalMesh))
  (bind _skeletalmesh_2 (Variables|Default|GetSkeletalMesh))
  (|CallParentFunction)
  (Utilities|FlowControl|DelayUntilNextTick)
  (Transformation|AttachComponentToComponent _skeletalmesh_2 _returnvalue "None" "KeepRelative" "KeepRelative" "KeepRelative" true)
  (bind _asskeletal_mesh_component (Utilities|Casting|CastToSkeletalMeshComponent _returnvalue))
  (Components|SkinnedMesh|SetLeaderPoseComponent _skeletalmesh_2 _asskeletal_mesh_component false false)
  (bind (_index _bindexisvalid) (Class|FortCosmeticFunctionLibrary|GetVariantChannelActiveIndexForSlot "(TagName=\"LoadoutSlot.Backpack\")" (CosmeticLoadout|GetCosmeticLoadout) "(TagName=\"Cosmetics.Variant.Channel.Particle\")" (Utilities|Array|MakeArray "(TagName=\"Cosmetics.Variant.Property.Particle2\")" "(TagName=\"Cosmetics.Variant.Property.Particle1\")")))
  (if (and _bindexisvalid (== _index 0))
    (Rendering|Material|SetMaterial _skeletalmesh 0 "/BRCosmetics/Accessories/FORT_Backpacks/Backpack_M_MED_Future_Samurai_Summer/FX/Material/M_FutureSamuraiSummer_Eye_TriRed.M_FutureSamuraiSummer_Eye_TriRed")
    (Rendering|Material|SetMaterial _skeletalmesh 1 "/BRCosmetics/Accessories/FORT_Backpacks/Backpack_M_MED_Future_Samurai_Summer/FX/Material/M_FutureSamuraiSummer_Eye_Double_Red.M_FutureSamuraiSummer_Eye_Double_Red")
    (Rendering|Material|SetMaterial _skeletalmesh 2 "/BRCosmetics/Accessories/FORT_Backpacks/Backpack_M_MED_Future_Samurai_Summer/FX/Material/M_FutureSamuraiSummer_Eye_Double_C_Red.M_FutureSamuraiSummer_Eye_Double_C_Red")
    (Rendering|Material|SetMaterial _skeletalmesh 3 "/BRCosmetics/Accessories/FORT_Backpacks/Backpack_M_MED_Future_Samurai_Summer/FX/Material/M_FutureSamuraiSummer_Eye_Double_B_Red.M_FutureSamuraiSummer_Eye_Double_B_Red")
    (Rendering|Material|SetMaterial _skeletalmesh 4 "/BRCosmetics/Accessories/FORT_Backpacks/Backpack_M_MED_Future_Samurai_Summer/FX/Material/M_FutureSamuraiSummer_Eye_One_Red.M_FutureSamuraiSummer_Eye_One_Red")))

(event EventOnDayPhaseChanged
  (params (CurrentDayPhase :type EFortDayPhase%20Enum :default "Morning") (PreviousDayPhase :type EFortDayPhase%20Enum :default "Morning") (bAtCreation :type Boolean)))

(event EventOnNewPlayerWeaponType
  (params (NewType :type EFortItemType%20Enum :default "WorldItem") (OldType :type EFortItemType%20Enum :default "WorldItem"))
  (|CallParentFunction NewType OldType) ; This node is disabled and will not be called.\nDrag off pins to build functionality.
)

(event EventOnWeaponChanged
  (params (NewWeapon :type Fort%20Weapon%20Object%20Reference) (OldWeapon :type Fort%20Weapon%20Object%20Reference)))

(event EventOnWeaponFired)

(event EventOnNewWeapons
  (params (WeaponsAsItems :type Array%20of%20Fort%20Item%20Object%20References)))

(event EventOnCurieStateAttached
  (params (Tag :type Gameplay%20Tag%20Structure%20%28by%20ref%29)))

(event EventOnCurieStateDetached
  (params (Tag :type Gameplay%20Tag%20Structure%20%28by%20ref%29)))

(event EventOnBeginSkydiving
  (|CallParentFunction) ; This node is disabled and will not be called.\nDrag off pins to build functionality.
)

(event EventOnEndSkydiving
  (|CallParentFunction) ; This node is disabled and will not be called.\nDrag off pins to build functionality.
)

(event EventOnEnterVehicle
  (params (VehicleEntered :type Fort%20Vehicle%20Interface%20Interface%20%28by%20ref%29))
  (|CallParentFunction VehicleEntered) ; This node is disabled and will not be called.\nDrag off pins to build functionality.
)

(event EventOnExitVehicle
  (|CallParentFunction) ; This node is disabled and will not be called.\nDrag off pins to build functionality.
)

(event EventOnTargetingOrFirstPersonCameraChanged)

(event EventOnHitPawn
  (params (SelfActor :type Actor%20Object%20Reference) (OtherActor :type Actor%20Object%20Reference) (NormalImpulse :type Vector :default "0, 0, 0") (Hit :type Hit%20Result%20Structure%20%28by%20ref%29)))

(fn ConstructionScript
  (|CallParentFunction))
