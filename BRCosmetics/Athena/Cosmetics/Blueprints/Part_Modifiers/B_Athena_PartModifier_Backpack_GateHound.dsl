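; Cosmetic part modifier for the GateHound back bling that drives its idle VFX from the selected progressive variant stage and a kill count.
; After res-in and a short retriggerable delay, it reads the loadout's stage index and passes it to the Niagara component as a float "Variant".
; It also binds to the backpack AnimBP's KillCountChanged event, forwarding each new count to the VFX "KillCount" parameter.
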(event EventOnPostResIn
  ;[0.1 0.1 0.1] Init
  (|CallParentFunction)
  (Utilities|FlowControl|RetriggerableDelay 0.1)
  (bind _index (Class|FortCosmeticFunctionLibrary|GetVariantChannelActiveIndexForSlot "(TagName=\"LoadoutSlot.Backpack\")" (CosmeticLoadout|GetCosmeticLoadout) "(TagName=\"Cosmetics.Variant.Channel.Progressive\")" (Utilities|Array|MakeArray "(TagName=\"Cosmetics.Variant.Property.Stage1\")" "(TagName=\"Cosmetics.Variant.Property.Stage2\")")))
  (bind _asniagara_particle_system_component (Utilities|Casting|CastToNiagaraComponent (CallFunction|GetIdleVFXComponent)))
  (Niagara|SetNiagaraVariable(Float) _asniagara_particle_system_component "Variant" (Math|Conversions|ToFloat(Integer) _index))
  (bind _asm_med_gate_hound_pack_anim_bp (Utilities|Casting|CastToM_MED_GateHound_Pack_AnimBP (Components|SkeletalMesh|GetAnimInstance (Customization|GetSkeletalMeshforPartType (CustomCharacter|GetAssociatedPlayerPawn) (CustomCharacter|GetCharacterPartType) :!class /Script/FortniteGame.FortPlayerPawn))))
  (Default|BindEventtoKillCountChanged _asm_med_gate_hound_pack_anim_bp (delegate Custom|KillCountChanged)))

(event Custom|KillCountChanged
  (params (KillCount :type Integer))
  ;[1 1 1] Reactivity
  (Effects|Components|ParticleSystem|SetIntParameter (CallFunction|GetIdleVFXComponent) "KillCount" KillCount))

(fn ConstructionScript
  (|CallParentFunction))
