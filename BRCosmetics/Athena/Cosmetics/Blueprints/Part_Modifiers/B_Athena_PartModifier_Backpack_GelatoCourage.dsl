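; Cosmetic part modifier for the GelatoCourage backpack that turns it from silver to gold when a watched stat rises.
; On post-res init in gameplay it records the Victory Crown total; a higher stat value triggers the transition, idle VFX reactivity and an attached sound.
; A delay-stepped timer drives two offset material scalars for base and spiral gold.
; The front end replays the effect on a repeating timer without sound.
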(event EventOnStatChanged
  (params (State%20Name :type Name :default "None") (Stat%20Value :type Integer))
  (bind _returnvalue (CustomCharacter|GetSkeletalMeshForPartOnPlayerPawn))
  (bind _returnvalue_2 (Utilities|Time|GetWorldDeltaSeconds))
  (bind _returnvalue_3 (CallFunction|GetIdleVFXComponent))
  ;[0 0 0] Stat Changed
  (if (and (> Stat%20Value 0) (> Stat%20Value (Variables|Default|GetInitialWinCount)))
    (Utilities|FlowControl|Sequence
      (:then_0
        ;[1 0.7876 0.2628] Backpack Material Change (Silver to Gold)
        (Variables|Default|SetMaterialTimer 0.0)
        (Variables|Default|SetGoldenTransition_Amount 0.0)
        (Utilities|FlowControl|Delay _returnvalue_2)
        (bind _output_get (Variables|Default|SetMaterialTimer (+ _returnvalue_2 (Variables|Default|GetMaterialTimer))))
        (bind _output_get_2 (Variables|Default|SetGoldenTransition_Amount (+ _returnvalue_2 (Variables|Default|GetGoldenTransition_Amount))))
        (Rendering|Material|SetScalarParameterValueonMaterials _returnvalue "GoldenAmount_Base" (Math|Float|Clamp(Float) (* (- _output_get_2 0.28) 1.1)))
        (Rendering|Material|SetScalarParameterValueonMaterials _returnvalue "GoldenAmount_Spyral" (Math|Float|Clamp(Float) (* (- _output_get_2 0.1) 1.05)))
        (if (> _output_get 4.0)
          _
          (else
            _)))
      (:then_1
        (Effects|Components|ParticleSystem|SetBoolParameter _returnvalue_3 "bReactivity" true)
        (Components|Activation|Activate _returnvalue_3 true)
        (CallFunction|PlaySound (CustomCharacter|GetSkeletalMeshForPartOnPlayerPawn))))))

(event Custom|TriggerReactivity
  (bind _returnvalue (CustomCharacter|GetSkeletalMeshForPartOnPlayerPawn))
  (bind _returnvalue_2 (Utilities|Time|GetWorldDeltaSeconds))
  (bind _returnvalue_3 (CallFunction|GetIdleVFXComponent))
  ;[0 0 0] Stat Changed
  (Utilities|FlowControl|Sequence
    (:then_0
      ;[1 0.7876 0.2628] Backpack Material Change (Silver to Gold)
      (Variables|Default|SetMaterialTimer 0.0)
      (Variables|Default|SetGoldenTransition_Amount 0.0)
      (Utilities|FlowControl|Delay _returnvalue_2)
      (bind _output_get (Variables|Default|SetMaterialTimer (+ _returnvalue_2 (Variables|Default|GetMaterialTimer))))
      (bind _output_get_2 (Variables|Default|SetGoldenTransition_Amount (+ _returnvalue_2 (Variables|Default|GetGoldenTransition_Amount))))
      (Rendering|Material|SetScalarParameterValueonMaterials _returnvalue "GoldenAmount_Base" (Math|Float|Clamp(Float) (* (- _output_get_2 0.28) 1.1)))
      (Rendering|Material|SetScalarParameterValueonMaterials _returnvalue "GoldenAmount_Spyral" (Math|Float|Clamp(Float) (* (- _output_get_2 0.1) 1.05)))
      (if (> _output_get 4.0)
        _
        (else
          _)))
    (:then_1
      (Effects|Components|ParticleSystem|SetBoolParameter _returnvalue_3 "bReactivity" true)
      (Components|Activation|Activate _returnvalue_3 true)
      (CallFunction|PlaySound (CustomCharacter|GetSkeletalMeshForPartOnPlayerPawn)))))

(event U
  (params (Key :type Key%20Structure)))

(event EventOnPostResIn
  ;[1 1 1] Initialize
  (|CallParentFunction)
  ;[1 1 1] Front end reactivity cycle
  (if (CustomCosmeticModifier|IsinFrontEnd)
    (Utilities|FlowControl|Delay 2.0)
    (CallFunction|TriggerReactivity)
    (Utilities|Time|SetTimerbyEvent (EventDispatchers|CreateEvent :!event Custom|TriggerReactivity) 11.0 true)
    (else
      ;[1 1 1] Setting initial state during gameplay
      (bind _returnvalue (Class|FortCosmeticFunctionLibrary|GetCosmeticStatValue 0 "(TagName=\"StatManager.TotalVictoryCrowns\")"))
      (Variables|Default|SetInitialWinCount _returnvalue))))

(fn PlaySound
  (params (AttachToComponent :type Scene%20Component%20Object%20Reference))
  (bind (_asfort_game_mode_front_end _bsuccess) (Utilities|Casting|CastToFortGameModeFrontEnd (Game|GetGameMode)))
  (if (not _bsuccess)
    (Audio|SpawnSoundAttached 0 AttachToComponent)))

(fn ConstructionScript
  (|CallParentFunction))
