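; Part modifier for the Genius backpack cosmetic that drives the bIsJump bool on its idle VFX from the wearer's airborne state.
; After res-in, it runs ToonFrontendLUTReplace, then binds the associated Athena pawn's jump and landed events to set and clear the flag.
; A short looping timer also copies the movement component's IsFalling into the flag, so it tracks falling as well as jumps.
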(event EventOnPostResIn
  (|CallParentFunction)
  (Utilities|ToonFrontendLUTReplace
    (:Out
      (bind _asfort_player_pawn_athena (Utilities|Casting|CastToFortPlayerPawnAthena (CustomCharacter|GetAssociatedPlayerPawn)))
      (Pawn|BindEventtoOnPawnJumped _asfort_player_pawn_athena (delegate Custom|Jump))
      (EventDispatchers|BindEventtoOnPawnLanded _asfort_player_pawn_athena (delegate Custom|Landed))
      (Utilities|Time|SetTimerbyEvent (delegate Custom|SecondCheck) 0.1 true false 0.0 0.1))))

(event Custom|SecondCheck
  (Effects|Components|ParticleSystem|SetBoolParameter (CallFunction|GetIdleVFXComponent) "bIsJump" (AI|Components|NavMovement|IsFalling (Pawn|GetMovementComponent (CustomCharacter|GetAssociatedPlayerPawn)))))

(event Custom|Jump
  (Effects|Components|ParticleSystem|SetBoolParameter (CallFunction|GetIdleVFXComponent) "bIsJump" true))

(event Custom|Landed
  (params (Hit :type Hit%20Result%20Structure%20%28by%20ref%29))
  (Effects|Components|ParticleSystem|SetBoolParameter (CallFunction|GetIdleVFXComponent) "bIsJump" false))

(fn ConstructionScript
  (|CallParentFunction))
