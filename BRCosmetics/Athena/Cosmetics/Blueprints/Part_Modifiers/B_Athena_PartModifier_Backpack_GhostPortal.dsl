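; Ghost Portal back bling part modifier that grows through four portal stages as the player's kill count rises.
; It attaches to the backpack root and watches the AthenaKills stat, setting a Niagara Stage parameter, replaying burst FX and fading in a funnel mesh at stage one.
; The frontend preview cycles all stages on delays.
; Stage one also starts a timeline feeding a world-position offset to a material, or destroys the actor if the pawn is dying.
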(event EventOnPostResIn
  ; FE Reactivity
  (|CallParentFunction)
  (if (CallFunction|ShouldDisplayFrontendReactivity)
    (Utilities|FlowControl|Sequence
      (:then_0
        (Utilities|FlowControl|Sequence
          (:then_0
            (CallFunction|CEStage0)
            (Utilities|FlowControl|Delay 3.0)
            (CallFunction|CEStage1)
            (Utilities|FlowControl|Delay 2.0)
            (CallFunction|CEStage2)
            (Utilities|FlowControl|Delay 2.0)
            (CallFunction|CEStage3))
          (:then_1
            (bind (_direction _opacity) (Timeline|FEFade_Timeline PlayFromStart)
              (:Update
                (Rendering|Material|SetScalarParameterValueonMaterials (Variables|Default|GetFront%20Plane) "BP_Opacity" _opacity))))))
      (:then_1
        (Utilities|FlowControl|Delay 0.35)))
    (else
      _))
  (Rendering|SetVisibility (Variables|Default|GetRear%20Plane) true false))

(event EventOnPawnAndPartChanged
  ;[0 0 0] BeginPlay
  (|CallParentFunction)
  (bind _returnvalue (Class|FortCosmeticFunctionLibrary|GetCosmeticStatValue (CustomCharacter|GetAssociatedPlayerPawn) "(TagName=\"StatManager.AthenaKills\")"))
  (Variables|Default|SetKillCount _returnvalue)
  (Transformation|AttachActorToComponent (Customization|GetSkeletalMeshforPartType (CustomCharacter|GetAssociatedPlayerPawn) (CustomCharacter|GetCharacterPartType) :!class /Script/FortniteGame.FortPlayerPawn) "backpack_root" "SnapToTarget" "SnapToTarget" "SnapToTarget")
  (CallFunction|UpdateBasedonKills)
  ;[0 0 0] Update Kills
  (if (> (stale Custom|UpdateBasedOnKills _statvalue) (Variables|Default|GetKillCount))
    _
    (else
      _))
  (bind _output_get (Variables|Default|SetKillCount (stale Custom|UpdateBasedOnKills _statvalue)))
  (if (>= _output_get 1)
    (if (>= _output_get 2)
      (if (>= _output_get 3)
        (CallFunction|CEStage3)
        (else
          (CallFunction|CEStage2)))
      (else
        (CallFunction|CEStage1)))
    (else
      (CallFunction|CEStage0))))

(event Custom|UpdateBasedOnKills
  ;[0 0 0] Update Kills
  (bind (_statname _statvalue) (Cosmetics|WatchCosmeticStat (CustomCharacter|GetAssociatedPlayerPawn) "(TagName=\"StatManager.AthenaKills\")" 0 false)
    (:OnStatChanged
      (if (> _statvalue (Variables|Default|GetKillCount))
        _
        (else
          _))
      (bind _output_get (Variables|Default|SetKillCount _statvalue))
      (if (>= _output_get 1)
        (if (>= _output_get 2)
          (if (>= _output_get 3)
            (CallFunction|CEStage3)
            (else
              (CallFunction|CEStage2)))
          (else
            (CallFunction|CEStage1)))
        (else
          (CallFunction|CEStage0))))))

(event Custom|CE_Stage0
  ;[0 0 0] Portal Stage
  (Effects|Components|ParticleSystem|SetFloatParameter (Variables|Default|GetNS_Backpack_GhostPortal_IdleFX) "Stage" 0.0)
  (Rendering|SetVisibility (Variables|Default|GetS1_Static%20Funnel) false false))

(event Custom|CE_Stage1
  ;[0 0 0] Portal Stage
  (Rendering|SetVisibility (Variables|Default|GetS1_Static%20Funnel) true false)
  (CallFunction|UpdateOffset)
  (Utilities|FlowControl|Sequence
    (:then_0
      (bind (_direction _opacity) (Timeline|FunnelTimeline PlayFromStart)
        (:Update
          (Rendering|Material|SetScalarParameterValueonMaterials (Variables|Default|GetS1_Static%20Funnel) "BP_Opacity" _opacity))))
    (:then_1
      (Effects|Components|ParticleSystem|SetFloatParameter (Variables|Default|GetNS_Backpack_GhostPortal_IdleFX) "Stage" 1.0)
      (Niagara|ReinitializeSystem (Variables|Default|GetNS_Backpack_GhostPortal_BurstFX)))))

(event Custom|CE_Stage2
  ;[0 0 0] Portal Stage
  (Effects|Components|ParticleSystem|SetFloatParameter (Variables|Default|GetNS_Backpack_GhostPortal_IdleFX) "Stage" 2.0)
  (Niagara|ReinitializeSystem (Variables|Default|GetNS_Backpack_GhostPortal_BurstFX)))

(event Custom|CE_Stage3
  ;[0 0 0] Portal Stage
  (Effects|Components|ParticleSystem|SetFloatParameter (Variables|Default|GetNS_Backpack_GhostPortal_IdleFX) "Stage" 3.0)
  (Niagara|ReinitializeSystem (Variables|Default|GetNS_Backpack_GhostPortal_BurstFX)))

(event Custom|Update%20Offset
  (bind _returnvalue (CustomCharacter|GetAssociatedPlayerPawn))
  (bind _returnvalue_2 (Transformation|GetActorLocation _returnvalue))
  (bind _returnvalue_3 (+ _returnvalue_2 "0, 0,50.000000"))
  ;[0 0 0] UpdateWorldPosMaterial
  (Variables|Default|SetLocationB _returnvalue_3)
  (Variables|Default|SetLocationA _returnvalue_3)
  (if (Variables|/Script/FortniteGame.FortPawn|GetbIsDying _returnvalue)
    (Actor|DestroyActor)
    (else
      (bind (_direction _lerp) (Timeline|Offset%20Controller PlayFromStart)
        (:Update
          (Rendering|Material|SetVectorParameterValue (Variables|Default|GetMID_WPO) "BP_OffsetPos" (Math|Conversions|ToLinearColor(Vector) (- (Math|Vector|Lerp(Vector) (Variables|Default|GetLocationB) (Variables|Default|GetLocationA) _lerp) (Transformation|GetActorLocation))) :!class /Script/Engine.MaterialInstanceDynamic))
        (:Finished
          (Variables|Default|SetLocationB (Variables|Default|GetLocationA)))))))

(fn ConstructionScript)
