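; Character part modifier for the male gnome backpack that makes its emissive material and idle VFX react to time of day.
; On post res-in it creates a dynamic material, then sets a daytime unlit or nighttime warm-glow state from the current hour.
; Day phase changes drive a DoOnce-gated timeline that fades the emissive color and toggles the VFX.
; In the front end, a looping timer flip-flops a second timeline to cycle the effect.
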(event EventOnPostResIn
  (bind _returnvalue (CallFunction|GetIdleVFXComponent))
  (bind _timeline_0 (Variables|B_Athena_PartModifier_Backpack_GnomeMale|GetTimeline_0))
  (|CallParentFunction)
  (Utilities|FlowControl|Sequence
    (:then_0
      (bind _returnvalue_2 (Rendering|Material|CreateDynamicMaterialInstance (CustomCharacter|GetSkeletalMeshForPartOnPlayerPawn) 0 (Rendering|Material|GetMaterial (CustomCharacter|GetSkeletalMeshForPartOnPlayerPawn) 0) "None" :!class /Script/Engine.PrimitiveComponent))
      (bind _output_get (Variables|Default|SetMID _returnvalue_2))
      (if (Math|Float|InRange(Float) (TimeOfDay|GetTimeOfDay :!class /Script/FortniteGame.FortKismetLibrary) 8.0 2e+01)
        (Rendering|Material|SetVectorParameterValue :self _output_get :ParameterName "Emissive Color" :!class /Script/Engine.MaterialInstanceDynamic)
        (Components|Activation|Deactivate _returnvalue)
        (Components|Timeline|SetNewTime _timeline_0 1.0)
        (else
          (Rendering|Material|SetVectorParameterValue _output_get "Emissive Color" "(R=2.000000,G=0.429234,B=0.075588,A=1.000000)" :!class /Script/Engine.MaterialInstanceDynamic)
          (Components|Activation|Activate _returnvalue false)
          (Components|Timeline|SetNewTime _timeline_0 0.0))))
    (:then_1
      (if (CustomCosmeticModifier|IsinFrontEnd)
        (Utilities|FlowControl|Delay 0.2)
        (Utilities|Time|SetTimerbyEvent (EventDispatchers|CreateEvent :!event Custom|FrontEnd_Event) 5.0 true)))))

(event EventOnDayPhaseChanged
  (params (CurrentDayPhase :type EFortDayPhase%20Enum :default "Morning") (PreviousDayPhase :type EFortDayPhase%20Enum :default "Morning") (bAtCreation :type Boolean))
  (bind _returnvalue (CallFunction|GetIdleVFXComponent))
  ;[1 1 1] TimeOfDay_Reactivity
  (switch Utilities|FlowControl|Switch|SwitchonEFortDayPhase CurrentDayPhase
    (:Morning
      _)
    (:Day
      _)
    (:Evening
      _)
    (:Night
      (Utilities|FlowControl|Sequence
        (:then_0
          (Utilities|FlowControl|DoOnce false :!label _label_2
            (:Completed
              (Components|Activation|Deactivate _returnvalue)
              (Timeline|Timeline_0 PlayFromStart))))
        (:then_1
          (Utilities|FlowControl|DoOnce false :!label _label
            (:Completed
              (Components|Activation|Activate _returnvalue false)
              (bind (_direction _thecolor) (Timeline|Timeline_0 Reverse)
                (:Update
                  (Rendering|Material|SetVectorParameterValue (Variables|Default|GetMID) "Emissive Color" _thecolor :!class /Script/Engine.MaterialInstanceDynamic)))))))))
  (Utilities|FlowControl|Sequence
    (:then_0
      (goto _label Reset))
    (:then_1
      (goto _label_2))))

(event Custom|FrontEnd_Event
  (bind _returnvalue (CallFunction|GetIdleVFXComponent))
  ;[1 1 1] FrontEnd_Cycling
  (Utilities|FlowControl|FlipFlop
    (:A
      (Utilities|FlowControl|Sequence
        (:then_0
          (Utilities|FlowControl|DoOnce false :!label _label
            (:Completed
              (Effects|Components|ParticleSystem|SetBoolParameter _returnvalue "FrontEndToggle" false)
              (bind (_direction _thecolor) (Timeline|Timeline_1 PlayFromStart)
                (:Update
                  (Rendering|Material|SetVectorParameterValue (Variables|Default|GetMID) "Emissive Color" _thecolor :!class /Script/Engine.MaterialInstanceDynamic))
                (:Finished
                  (switch Utilities|FlowControl|Switch|SwitchonETimelineDirection _direction
                    (:Forward
                      (Components|Activation|Deactivate _returnvalue))))))))
        (:then_1
          _)))
    (:B
      (Utilities|FlowControl|Sequence
        (:then_0
          _)
        (:then_1
          (goto _label)))))
  (Utilities|FlowControl|DoOnce false
    (:Completed
      (Components|Activation|Activate _returnvalue false)
      (Effects|Components|ParticleSystem|SetBoolParameter _returnvalue "FrontEndToggle" true)
      (Timeline|Timeline_1 Reverse))))

(fn ConstructionScript)
