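; Part modifier for the Grapefruit back bling, a Pac-Man themed sprite effect that reacts to the wearer.
; Setup reads the particle variant, then binds damage, DBNO and a frontend preview timer that alternates facing.
; Tick turns local acceleration into a facing, driving sprite-sheet column offsets on the idle VFX.
; Damage and landing play sounds and glitch timelines; nonzero variants show blue-ghost frames when hurt and separate cells when DBNO.
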(event EventOnPostResIn
  (|CallParentFunction)
  (CallFunction|SetupGrapefruit))

(event Custom|SetupGrapefruit
  (bind _variant (Variables|Default|Getvariant))
  (bind _returnvalue (> _variant 0))
  (bind _variant_2 (Variables|Default|Getvariant))
  (bind _returnvalue_2 (Math|Conversions|ToFloat(Integer) _variant_2))
  (bind _returnvalue_3 (CallFunction|GetIdleVFXComponent))
  (bind _returnvalue_4 (CustomCharacter|GetAssociatedPlayerPawn))
  (Utilities|FlowControl|Sequence
    (:then_0
      ;[1 1 1] Init
      (bind _index (CosmeticLoadout|GetVariantChannelActiveIndexforSlotActor "(TagName=\"LoadoutSlot.Backpack\")" _returnvalue_4 "(TagName=\"Cosmetics.Variant.Channel.Particle\")" (Utilities|Array|MakeArray "(TagName=\"Cosmetics.Variant.Property.Particle1\")" "(TagName=\"Cosmetics.Variant.Property.Particle2\")" "(TagName=\"Cosmetics.Variant.Property.Particle3\")" "(TagName=\"Cosmetics.Variant.Property.Particle4\")" "(TagName=\"Cosmetics.Variant.Property.Particle5\")")))
      (bind _output_get (Variables|Default|Setvariant _index))
      (if (> _output_get 0)
        (Variables|Default|Setcolumn_offset 2.0)
        (else
          (Variables|Default|Setcolumn_offset 4.0)))
      (Utilities|Casting|CastToFortPlayerPawnAthena _returnvalue_4)
      (EventDispatchers|BindEventtoOnDBNOStateChanged _returnvalue_4 (delegate Custom|CustomEvent_0)))
    (:then_1
      ;[1 1 1] Damage Received
      (bind (_statname _statvalue) (Cosmetics|WatchCosmeticStat (CustomCharacter|GetAssociatedPlayerPawn) "(TagName=\"StatManager.DamageReceived\")" 0 false)
        (:OnStatChanged
          (if (> _statvalue 0)
            (Development|PrintString (Utilities|String|Append "Grapefruit " (Utilities|String|ToString(Name) _statname) " " (Utilities|String|ToString(Integer) _statvalue)))
            (if (not (Variables|Default|GetDBNO))
              (CallFunction|PlaySound (Variables|Default|GetDamagedSound))
              (bind (_direction _glitch) (Timeline|BlueGhostGlitch PlayFromStart)
                (:Update
                  (Effects|Components|ParticleSystem|SetFloatParameter _returnvalue_3 "glitch_amount" _glitch))
                (:BlueGhostEventStart
                  (if _returnvalue
                    (Effects|Components|ParticleSystem|SetFloatParameter _returnvalue_3 "row_offset" 5.0)
                    (Effects|Components|ParticleSystem|SetFloatParameter _returnvalue_3 "column_offset" 0.0)
                    (Effects|Components|ParticleSystem|SetFloatParameter _returnvalue_3 "use_color" 0.0)
                    (Variables|Default|Setblue_ghost_active true)))
                (:BlueGhostEventEnd
                  (if _returnvalue
                    (Development|PrintString (Utilities|String|Append "pacman row_offfset " (Utilities|String|ToString(Float) _returnvalue_2)))
                    (Effects|Components|ParticleSystem|SetFloatParameter _returnvalue_3 "row_offset" _returnvalue_2)
                    (Effects|Components|ParticleSystem|SetFloatParameter _returnvalue_3 "column_offset" (Variables|Default|Getcolumn_offset))
                    (Effects|Components|ParticleSystem|SetFloatParameter _returnvalue_3 "use_color" 1.0)
                    (Variables|Default|Setblue_ghost_active false)))))))))
    (:then_2
      ;[1 1 1] FrontEnd
      (if (Reactive|Frontend|IsCosmeticObjectDisplayedasPreview (CustomCharacter|GetAssociatedPlayerPawn))
        (Utilities|FlowControl|Delay 0.2)
        (Utilities|Time|SetTimerbyEvent (delegate Custom|FrontEndReactivity) 3.0 true)))))

(event Custom|CustomEvent_0
  (params (FortPawn :type Fort%20Pawn%20Object%20Reference) (bInIsDBNO :type Boolean))
  (bind _returnvalue (CallFunction|GetIdleVFXComponent))
  (bind _variant (Variables|Default|Getvariant))
  (bind _returnvalue_2 (> _variant 0))
  (bind _variant_2 (Variables|Default|Getvariant))
  (bind _returnvalue_3 (Math|Conversions|ToFloat(Integer) _variant_2))
  (bind _returnvalue_4 (CallFunction|GetIdleVFXComponent))
  (bind _variant_3 (Variables|Default|Getvariant))
  ;[1 1 1] DBNO
  (bind _output_get (Variables|Default|SetDBNO bInIsDBNO))
  (Utilities|FlowControl|Sequence
    (:then_0
      (Variables|Default|Setblue_ghost_active false)
      (if (> (Variables|Default|Getvariant) 0)
        (if bInIsDBNO
          (Utilities|FlowControl|Sequence
            (:then_0
              (Effects|Components|ParticleSystem|SetFloatParameter _returnvalue "row_offset" 5.0)
              (Effects|Components|ParticleSystem|SetFloatParameter _returnvalue "column_offset" 5.0)
              (Effects|Components|ParticleSystem|SetFloatParameter _returnvalue "cell_count" 1.0)
              (Effects|Components|ParticleSystem|SetFloatParameter _returnvalue "use_color" 0.0))
            (:then_1
              ;[1 1 1] Damage Received
              (Timeline|BlueGhostGlitch Stop)))
          (else
            (Effects|Components|ParticleSystem|SetFloatParameter _returnvalue "row_offset" (Math|Conversions|ToFloat(Integer) _variant_3))
            (bind _result (ArrayMacros|SafeArrayGet (Variables|Default|Getcolumn_offset_defaults) _variant_3)
              (:%20%20
                (Effects|Components|ParticleSystem|SetFloatParameter _returnvalue "column_offset" _result)
                (Effects|Components|ParticleSystem|SetFloatParameter _returnvalue "cell_count" 2.0)
                (Effects|Components|ParticleSystem|SetFloatParameter _returnvalue "use_color" 1.0)))))))
    (:then_1
      (if _output_get
        (if (>= (Variables|Default|Getvariant) 1)
          (CallFunction|PlaySound (Variables|Default|GetDBNOGhostSound))
          (else
            (CallFunction|PlaySound (Variables|Default|GetDBNOSound))))))))

(event EventOnPawnLanded
  ;[1 1 1] Landing Glitch
  (CallFunction|PlaySound (Variables|Default|GetLandSound))
  (bind (_direction _glitch) (Timeline|Glitch PlayFromStart)
    (:Update
      (Effects|Components|ParticleSystem|SetFloatParameter (CallFunction|GetIdleVFXComponent) "glitch_amount" _glitch))))

(event Custom|FrontEndReactivity
  ;[1 1 1] FrontEnd
  (if (> (Variables|Default|Getvariant) 0)
    (if (> (Variables|Default|Getcolumn_offset) 4.0)
      (Variables|Default|Setcolumn_offset 2.0)
      (else
        (Variables|Default|Setcolumn_offset 6.0)))
    (Effects|Components|ParticleSystem|SetFloatParameter (CallFunction|GetIdleVFXComponent) "column_offset" (Variables|Default|Getcolumn_offset))))

(event EventTick
  (params (DeltaSeconds :type Float%20%28single-precision%29))
  (bind _returnvalue (CallFunction|GetIdleVFXComponent))
  (bind _direction (Variables|Default|Getdirection))
  ;[1 1 1] Local Acceleration Direction
  (|CallParentFunction DeltaSeconds)
  (if (not (Variables|Default|Getblue_ghost_active))
    (Utilities|FlowControl|Sequence
      (:then_0
        (bind (_direction_2 _direction_amount) (CallFunction|GetBasisDirection (Math|Transform|InverseTransformDirection (Transformation|GetActorTransform (CustomCharacter|GetAssociatedPlayerPawn)) (Pawn|Components|CharacterMovement|GetCurrentAcceleration (Variables|/Script/Engine.Character|GetCharacterMovement (CustomCharacter|GetAssociatedPlayerPawn))))))
        (Variables|Default|Setdirection _direction_2)
        (Variables|Default|Setdirection_amount _direction_amount))
      (:then_1
        ;[1 1 1] Set Niagara Params
        (if (>= (Variables|Default|Getdirection_amount) 0.8)
          (if (> (Variables|Default|Getvariant) 0)
            (if (Variables|Default|GetDBNO)
              (Effects|Components|ParticleSystem|SetFloatParameter _returnvalue "column_offset" (Math|Conversions|ToFloat(Integer) (+ _direction 4)))
              (else
                (bind _output_get (Variables|Default|Setcolumn_offset (* _direction 2.0)))
                (Effects|Components|ParticleSystem|SetFloatParameter _returnvalue "column_offset" _output_get)))
            (else
              (switch int _direction
                (:1
                  (Effects|Components|ParticleSystem|SetFloatParameter _returnvalue "column_offset" 4.0))
                (:3
                  (Effects|Components|ParticleSystem|SetFloatParameter _returnvalue "column_offset" 0.0))))))))))

(fn GetBasisDirection
  (params (Local%20Player%20Acceleration :type Vector))
  (bind _returnvalue (Math|Vector|Normalize Local%20Player%20Acceleration))
  (bind _x (.x _returnvalue))
  (bind _y (.y _returnvalue))
  (bind _returnvalue_2 (Math|Float|Absolute(Float) _y))
  (bind _returnvalue_3 (Math|Float|Absolute(Float) _x))
  (if (> _returnvalue_3 _returnvalue_2)
    (LocalVariables|Setmy_direction_amount _returnvalue_3)
    (if (>= _x 0.0)
      (LocalVariables|Setmy_direction 0)
      (else
        (LocalVariables|Setmy_direction 2)))
    (else
      (LocalVariables|Setmy_direction_amount _returnvalue_2)
      (if (>= _y 0.0)
        (LocalVariables|Setmy_direction 1)
        (else
          (LocalVariables|Setmy_direction 3)))))
  (return (LocalVariables|Getmy_direction) (LocalVariables|Getmy_direction_amount)))

(fn PlaySound
  (params (NewSound :type Sound%20Base%20Object%20Reference))
  (Audio|SpawnSoundAttached NewSound (CustomCharacter|GetSkeletalMeshForPartOnPlayerPawn)))

(fn ConstructionScript
  (|CallParentFunction))
