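; Part modifier for the Grilled Cheese back bling that steps through four visual stages as the player's kill count rises.
; Pawn or part changes map the AthenaKills stat to a stage via a reverse threshold search, and stat activations set it directly with a burst effect.
; Stage events set backpack material scalars and idle VFX parameters.
; In the front end, a timeline previews all stages, feeding fake kill counts to the backpack AnimBP.
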(event EventOnPawnAndPartChanged
  (|CallParentFunction)
  (if (not (CustomCosmeticModifier|IsinFrontEnd))
    (bind _returnvalue (Class|FortCosmeticFunctionLibrary|GetCosmeticStatValue (CustomCharacter|GetAssociatedPlayerPawn) "(TagName=\"StatManager.AthenaKills\")"))
    (bind _stage (CallFunction|StatValuetoStage _returnvalue))
    (Variables|Default|SetCurrentStage _stage)
    (CallFunction|UpdateMaterialParameters)))

(event EventOnPostResIn
  (bind _animbp (Variables|Default|GetAnimBP))
  (|CallParentFunction)
  (if (CustomCosmeticModifier|IsinFrontEnd)
    (bind _asm_med_grilled_cheese_pack_anim_bp (Utilities|Casting|CastToM_MED_Grilled_Cheese_Pack_AnimBP (Components|SkeletalMesh|GetAnimInstance (Customization|GetSkeletalMeshforPartType (CustomCharacter|GetAssociatedPlayerPawn) (CustomCharacter|GetCharacterPartType) :!class /Script/FortniteGame.FortPlayerPawn))))
    (Variables|Default|SetAnimBP _asm_med_grilled_cheese_pack_anim_bp)
    (Timeline|Timeline_0 PlayFromStart
      (:Stage%201
        (Class|MMEDGrilledCheesePackAnimBP|UpdateFEKillCountEvent _animbp 0.0)
        (CallFunction|StageOne))
      (:Stage%202
        (Class|MMEDGrilledCheesePackAnimBP|UpdateFEKillCountEvent _animbp 1.0)
        (CallFunction|StageTwo))
      (:Stage%203
        (Class|MMEDGrilledCheesePackAnimBP|UpdateFEKillCountEvent _animbp 3.0)
        (CallFunction|StageThree))
      (:Stage%204
        (Class|MMEDGrilledCheesePackAnimBP|UpdateFEKillCountEvent _animbp 6.0)
        (CallFunction|StageFour)))
    (PartModifier|SpawnSystem|SpawnSystemAttachedPartModifier "/BRCosmetics/Accessories/FORT_Backpacks/Backpack_M_MED_Grilled_Cheese/FX/NS_Backpack_GrilledCheese_Burst.NS_Backpack_GrilledCheese_Burst" (CustomCharacter|GetSkeletalMeshForPartOnPlayerPawn) "FXsocket")
    (else
      (CallFunction|UpdateMaterialParameters))))

(event EventOnCosmeticStateActivated
  (params (Stat%20Name :type Name :default "None") (Stat%20Value :type Integer) (STAGE%20Num%20Being%20Activated :type Integer))
  (|CallParentFunction Stat%20Name Stat%20Value STAGE%20Num%20Being%20Activated)
  (Variables|Default|SetCurrentStage STAGE%20Num%20Being%20Activated)
  (CallFunction|UpdateMaterialParameters)
  (PartModifier|SpawnSystem|SpawnSystemAttachedPartModifier "/BRCosmetics/Accessories/FORT_Backpacks/Backpack_M_MED_Grilled_Cheese/FX/NS_Backpack_GrilledCheese_Burst.NS_Backpack_GrilledCheese_Burst" (CustomCharacter|GetSkeletalMeshForPartOnPlayerPawn) "root"))

(event Custom|StageOne
  (bind _returnvalue (CustomCharacter|GetSkeletalMeshForPartOnPlayerPawn))
  (bind _returnvalue_2 (CallFunction|GetIdleVFXComponent))
  (Effects|Components|ParticleSystem|SetFloatParameter _returnvalue_2 "FinalStage" 0.0)
  (Effects|Components|ParticleSystem|SetFloatParameter _returnvalue_2 "MidStage" 0.0)
  ;[1 1 1] BackPack Parameters
  (Rendering|Material|SetScalarParameterValueonMaterials _returnvalue "GC_UseStage1" 1.0)
  (Rendering|Material|SetScalarParameterValueonMaterials _returnvalue "GC_FresnelExponent" 4.0)
  (Rendering|Material|SetScalarParameterValueonMaterials _returnvalue "GC_FresnelPower" 1.0)
  (Rendering|Material|SetScalarParameterValueonMaterials _returnvalue "GC_FresneBrightness" 1.0)
  (Rendering|Material|SetScalarParameterValueonMaterials _returnvalue "GC_Brightness_Max" 8.0)
  (Rendering|Material|SetScalarParameterValueonMaterials _returnvalue "GC_OffsetAmount" 0.0)
  (Rendering|Material|SetScalarParameterValueonMaterials _returnvalue "GC_UseStage2" 0.0)
  (Rendering|Material|SetScalarParameterValueonMaterials _returnvalue "GC_UseStage3" 0.0)
  (Rendering|Material|SetScalarParameterValueonMaterials _returnvalue "GC_UseStage4" 0.0))

(event Custom|StageTwo
  (bind _returnvalue (CustomCharacter|GetSkeletalMeshForPartOnPlayerPawn))
  (bind _returnvalue_2 (CallFunction|GetIdleVFXComponent))
  (Effects|Components|ParticleSystem|SetFloatParameter _returnvalue_2 "FinalStage" 0.0)
  (Effects|Components|ParticleSystem|SetFloatParameter _returnvalue_2 "MidStage" 0.0)
  ;[1 1 1] BackPack Parameters
  (Rendering|Material|SetScalarParameterValueonMaterials _returnvalue "GC_UseStage2" 1.0)
  (Rendering|Material|SetScalarParameterValueonMaterials _returnvalue "GC_FresnelExponent" 6.0)
  (Rendering|Material|SetScalarParameterValueonMaterials _returnvalue "GC_FresnelPower" 1.8)
  (Rendering|Material|SetScalarParameterValueonMaterials _returnvalue "GC_FresneBrightness" 0.25)
  (Rendering|Material|SetScalarParameterValueonMaterials _returnvalue "GC_Brightness_Max" 4.0)
  (Rendering|Material|SetScalarParameterValueonMaterials _returnvalue "GC_OffsetAmount" -0.2))

(event Custom|StageThree
  (bind _returnvalue (CustomCharacter|GetSkeletalMeshForPartOnPlayerPawn))
  (bind _returnvalue_2 (CallFunction|GetIdleVFXComponent))
  (Effects|Components|ParticleSystem|SetFloatParameter _returnvalue_2 "FinalStage" 0.0)
  (Effects|Components|ParticleSystem|SetFloatParameter _returnvalue_2 "MidStage" 1.0)
  ;[1 1 1] BackPack Parameters
  (Rendering|Material|SetScalarParameterValueonMaterials _returnvalue "GC_UseStage3" 1.0)
  (Rendering|Material|SetScalarParameterValueonMaterials _returnvalue "GC_FresnelExponent" 5.0)
  (Rendering|Material|SetScalarParameterValueonMaterials _returnvalue "GC_FresneBrightness" 0.35)
  (Rendering|Material|SetScalarParameterValueonMaterials _returnvalue "GC_Brightness_Max" 5.0)
  (Rendering|Material|SetScalarParameterValueonMaterials _returnvalue "GC_OffsetAmount" -0.5))

(event Custom|StageFour
  (bind _returnvalue (CallFunction|GetIdleVFXComponent))
  (bind _returnvalue_2 (CustomCharacter|GetSkeletalMeshForPartOnPlayerPawn))
  (Effects|Components|ParticleSystem|SetFloatParameter _returnvalue "FinalStage" 1.0)
  (Effects|Components|ParticleSystem|SetFloatParameter _returnvalue "MidStage" 1.0)
  ;[1 1 1] BackPack Parameters
  (Rendering|Material|SetScalarParameterValueonMaterials _returnvalue_2 "GC_UseStage4" 1.0)
  (Rendering|Material|SetScalarParameterValueonMaterials _returnvalue_2 "GC_FresnelExponent" 4.0)
  (Rendering|Material|SetScalarParameterValueonMaterials _returnvalue_2 "GC_FresneBrightness" 0.5)
  (Rendering|Material|SetScalarParameterValueonMaterials _returnvalue_2 "GC_Brightness_Max" 6.0)
  (Rendering|Material|SetScalarParameterValueonMaterials _returnvalue_2 "GC_OffsetAmount" -1.25))

(fn StatValueToStage
  (params (StatValue :type Integer))
  (bind (_arrayindex _arrayelement) (Utilities|Array|ReverseforEachLoop (Variables|PartModifierControls|WatchCosmeticStat|GetStat%20Thresholds))
    (:LoopBody
      (if (>= (Variables|GetStatValue) _arrayelement)
        (return _arrayindex)))
    (:Completed
      (return 0))))

(fn UpdateMaterialParameters
  (switch int (Variables|Default|GetCurrentStage)
    (:0
      (CallFunction|StageOne))
    (:1
      (CallFunction|StageTwo))
    (:2
      (CallFunction|StageThree))
    (:3
      (CallFunction|StageFour))))

(fn ConstructionScript
  (|CallParentFunction))
