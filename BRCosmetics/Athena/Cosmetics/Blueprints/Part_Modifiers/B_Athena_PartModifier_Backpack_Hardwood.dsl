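; Part modifier for the Hardwood backpack cosmetic, a basketball hoop that toy basketballs can score in.
; After res-in it attaches a trigger box to the part's skeletal mesh and enables overlaps after a short delay.
; An overlapping ball scores unless it hit another component first, playing burst particles and attached sound, then hiding and triggering its target hit.
; The score animation is material-driven, started by writing game time to a scalar parameter.
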(event EventOnPostResIn
  (|CallParentFunction)
  (Transformation|AttachComponentToComponent (Variables|Default|GetHoopTriggerBox) (CustomCharacter|GetSkeletalMeshForPartOnPlayerPawn) "None" "KeepRelative" "KeepRelative" "KeepRelative" true)
  (Utilities|FlowControl|Delay 0.2)
  (Variables|/Script/Engine.PrimitiveComponent|SetbGenerateOverlapEvents true (Variables|Default|GetHoopTriggerBox)))

(event OnComponentBeginOverlap(HoopTriggerBox)
  (params (OverlappedComponent :type Primitive%20Component%20Object%20Reference) (OtherActor :type Actor%20Object%20Reference) (OtherComp :type Primitive%20Component%20Object%20Reference) (OtherBodyIndex :type Integer) (bFromSweep :type Boolean) (SweepResult :type Hit%20Result%20Structure%20%28by%20ref%29))
  (bind _asb_toy_basketball_base (Utilities|Casting|CastToB_ToyBasketball_Base OtherActor))
  ;[1 1 1] Only score if the ball hasn't hit something else first
  (if (not (and (Utilities|IsValid (Variables|/Game/Abilities/Toys/Basketball/Shared/B_ToyBasketball_Base.B_ToyBasketball_Base_C|GetHitComponent _asb_toy_basketball_base) :!class /Script/Engine.KismetSystemLibrary) (Utilities|NotEqual(Object) (Variables|/Game/Abilities/Toys/Basketball/Shared/B_ToyBasketball_Base.B_ToyBasketball_Base_C|GetHitComponent _asb_toy_basketball_base) (Variables|Default|GetHoopTriggerBox))))
    (GameplayCue|PlayBurstEffect (Utilities|Struct|MakeFortBurstEffectData (Variables|Default|GetBurst%20Particles)) _asb_toy_basketball_base (Ability|GameplayCue|MakeGameplayCueParameters :NormalizedMagnitude 0.0 :RawMagnitude 0.0 :Location (Math|Vector|MakeVectorNetQuantize10 (.x (Transformation|GetActorLocation OtherActor)) (.y (Transformation|GetActorLocation OtherActor)) (.z (Transformation|GetActorLocation OtherActor))) :Normal "0,0,0" :GameplayEffectLevel 1 :AbilityLevel 1 :bReplicateLocationWhenUsingMinimalRepProxy false))
    ;[1 1 1] Spawning sound attached in order to make sure the sound moves with the player
    (Audio|SpawnSoundAttached "/Game/Athena/Sounds/Toys/Toys_BasketBall_Score_Net_Cue.Toys_BasketBall_Score_Net_Cue" (Variables|Default|GetHoopTriggerBox))
    (Rendering|SetActorHiddenInGame _asb_toy_basketball_base true)
    (Collision|SetActorEnableCollision _asb_toy_basketball_base false)
    (Class|FortPhysicsBall|AuthTriggerTargetHit _asb_toy_basketball_base (Transformation|GetWorldLocation (Variables|Default|GetHoopTriggerBox)) (+ (Transformation|GetWorldLocation (Variables|Default|GetHoopTriggerBox)) (* (Transformation|GetActorUpVector) -1.2e+02)))
    (CallFunction|PlayMIDScoreAnimation)))

(fn PlayMIDScoreAnimation
  (Rendering|Material|SetScalarParameterValueonMaterials (CustomCharacter|GetSkeletalMeshForPartOnPlayerPawn) "Play Score Animation" (Utilities|Time|GetGameTimeinSeconds)))

(fn ConstructionScript)
