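; Backpack part modifier for the HighTower Mango cosmetic that switches the pack's glow effect as its cosmetic stage changes.
; EventOnCosmeticStateActivated calls the parent, then routes stages 0-3 to custom events and prints any unhandled stage.
; Stage 0 deactivates all three stored glows; stages 1-3 each spawn and store their Niagara system at "pack", deactivating the other two.
; Debug "kill" prints remain, and a stage never deactivates its own earlier glow before respawning.
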(event Custom|Stage%200
  (bind _glow_3 (Variables|Default|GetGlow%203))
  (bind _glow_1 (Variables|Default|GetGlow%201))
  (bind _glow_2 (Variables|Default|GetGlow%202))
  (Development|PrintString "kill 0" true true "(R=1.000000,G=0.000000,B=0.000000,A=1.000000)" 5.0)
  (Utilities|IsValid _glow_1
    (:Is%20Valid
      (Components|Activation|Deactivate _glow_1))
    (:Is%20Not%20Valid
      _))
  (Utilities|IsValid _glow_2
    (:Is%20Valid
      (Components|Activation|Deactivate _glow_2))
    (:Is%20Not%20Valid
      _))
  (Utilities|IsValid _glow_3
    (:Is%20Valid
      (Components|Activation|Deactivate _glow_3))))

(event Custom|Stage%201
  (bind _glow_2 (Variables|Default|GetGlow%202))
  (bind _glow_3 (Variables|Default|GetGlow%203))
  (Development|PrintString "kill 1" true true "(R=1.000000,G=0.000000,B=0.000000,A=1.000000)" 1e+01)
  (bind _returnvalue (PartModifier|SpawnSystem|SpawnSystemAttachedPartModifier "/BRCosmetics/Accessories/FORT_Backpacks/Backpack_F_MED_HighTower_Mango/FX/NS_Backpack_HighTower_Mango_Stage1.NS_Backpack_HighTower_Mango_Stage1" (CustomCharacter|GetSkeletalMeshForPartOnPlayerPawn) "pack")
    (:Success
      (Variables|Default|SetGlow%201 _returnvalue))
    (:Failure
      _))
  (Utilities|IsValid _glow_2
    (:Is%20Valid
      (Components|Activation|Deactivate _glow_2))
    (:Is%20Not%20Valid
      _))
  (Utilities|IsValid _glow_3
    (:Is%20Valid
      (Components|Activation|Deactivate _glow_3))))

(event EventOnCosmeticStateActivated
  (params (Stat%20Name :type Name :default "None") (Stat%20Value :type Integer) (STAGE%20Num%20Being%20Activated :type Integer))
  (|CallParentFunction Stat%20Name Stat%20Value STAGE%20Num%20Being%20Activated)
  (switch int STAGE%20Num%20Being%20Activated
    (:Default
      (Development|PrintString (Utilities|String|Append "Unhandled Cosmetic Stage " (Utilities|String|ToString(Integer) STAGE%20Num%20Being%20Activated)) true true "(R=1.000000,G=0.000000,B=0.006820,A=1.000000)" 5.0))
    (:0
      (CallFunction|Stage0))
    (:1
      (CallFunction|Stage1))
    (:2
      (CallFunction|Stage2))
    (:3
      (CallFunction|Stage3))))

(event Custom|Stage%202
  (bind _glow_1 (Variables|Default|GetGlow%201))
  (bind _glow_3 (Variables|Default|GetGlow%203))
  (Development|PrintString "kill 3" true true "(R=1.000000,G=0.000000,B=0.000000,A=1.000000)" 1e+01)
  (Utilities|IsValid _glow_1
    (:Is%20Valid
      (Components|Activation|Deactivate _glow_1))
    (:Is%20Not%20Valid
      _))
  (bind _returnvalue (PartModifier|SpawnSystem|SpawnSystemAttachedPartModifier "/BRCosmetics/Accessories/FORT_Backpacks/Backpack_F_MED_HighTower_Mango/FX/NS_Backpack_HighTower_Mango_Stage2.NS_Backpack_HighTower_Mango_Stage2" (CustomCharacter|GetSkeletalMeshForPartOnPlayerPawn) "pack")
    (:Success
      (Variables|Default|SetGlow%202 _returnvalue))
    (:Failure
      _))
  (Utilities|IsValid _glow_3
    (:Is%20Valid
      (Components|Activation|Deactivate _glow_3))))

(event Custom|Stage%203
  (bind _glow_2 (Variables|Default|GetGlow%202))
  (bind _glow_1 (Variables|Default|GetGlow%201))
  (Development|PrintString "kill 5" true true "(R=1.000000,G=0.000000,B=0.000000,A=1.000000)" 1e+01)
  (Utilities|IsValid _glow_1
    (:Is%20Valid
      (Components|Activation|Deactivate _glow_1))
    (:Is%20Not%20Valid
      _))
  (Utilities|IsValid _glow_2
    (:Is%20Valid
      (Components|Activation|Deactivate _glow_2))
    (:Is%20Not%20Valid
      _))
  (bind _returnvalue (PartModifier|SpawnSystem|SpawnSystemAttachedPartModifier "/BRCosmetics/Accessories/FORT_Backpacks/Backpack_F_MED_HighTower_Mango/FX/NS_Backpack_HighTower_Mango_Stage3.NS_Backpack_HighTower_Mango_Stage3" (CustomCharacter|GetSkeletalMeshForPartOnPlayerPawn) "pack")
    (:Success
      (Variables|Default|SetGlow%203 _returnvalue))))

(fn ConstructionScript
  (|CallParentFunction))
