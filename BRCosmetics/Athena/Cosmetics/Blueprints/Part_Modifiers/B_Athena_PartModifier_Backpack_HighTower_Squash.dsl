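; Backpack part modifier for the HighTower Squash cape that plays material and Niagara transitions as the cape appears and when emote notifies remove it.
; Custom primitive data passes start and end times to transition materials, with source materials restored after the appear delay.
; Both notify events duplicate the fade-out and destroy the effect after a delay.
; GetVariantLevel reads the Progressive variant for material choice, but only Stage2 returns success, enabling swaps.
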(event Custom|AnimNotify_MaterialTransition
  (bind _returnvalue (CustomCharacter|GetSkeletalMeshForPartOnPlayerPawn))
  ;[1 1 1 1 40] Called from AnimMontage'/Game/Animation/Game/MainPlayer/Emotes/HighTowerSquash/Emote_HighTowerSquash_CMF_M.Emote_HighTowerSquash_CMF_M'
  (bind _returnvalue_2 (PartModifier|SpawnSystem|SpawnSystemAttachedPartModifier "/BRCosmetics/Accessories/FORT_Capes/Cape_F_MED_Hightower_Squash/FX/NS_Cape_HighTower_Squash.NS_Cape_HighTower_Squash" (CustomCharacter|GetSkeletalMeshForPartOnPlayerPawn) "root")
    (:Success
      (Variables|Default|SetNiagaraEffects _returnvalue_2))
    (:Failure
      _))
  ;;[1 1 1] Material Transition Out
  (CallFunction|ApplyTransitionMaterial false)
  (Rendering|Material|SetCustomPrimitiveDataFloat _returnvalue 0 0.0)
  (Rendering|Material|SetCustomPrimitiveDataFloat _returnvalue 1 (Utilities|Time|GetGameTimeinSeconds))
  (Utilities|FlowControl|Delay 0.5)
  (Components|DestroyComponent (Variables|Default|GetNiagaraEffects)))

(event Custom|AnimNotify_CapeOutTransitions
  (bind _returnvalue (CustomCharacter|GetSkeletalMeshForPartOnPlayerPawn))
  ;[1 1 1 1 40] Called from AnimMontage'/Game/Animation/Game/MainPlayer/Emotes/HighTowerSquash/Emote_HighTowerSquash_CMF_M.Emote_HighTowerSquash_CMF_M'
  (bind _returnvalue_2 (PartModifier|SpawnSystem|SpawnSystemAttachedPartModifier "/BRCosmetics/Accessories/FORT_Capes/Cape_F_MED_Hightower_Squash/FX/NS_Cape_HighTower_Squash.NS_Cape_HighTower_Squash" (CustomCharacter|GetSkeletalMeshForPartOnPlayerPawn) "root")
    (:Success
      (Variables|Default|SetNiagaraEffects _returnvalue_2))
    (:Failure
      _))
  ;;[1 1 1] Material Transition Out
  (CallFunction|ApplyTransitionMaterial false)
  (Rendering|Material|SetCustomPrimitiveDataFloat _returnvalue 0 0.0)
  (Rendering|Material|SetCustomPrimitiveDataFloat _returnvalue 1 (Utilities|Time|GetGameTimeinSeconds))
  (Utilities|FlowControl|Delay 0.5)
  (Components|DestroyComponent (Variables|Default|GetNiagaraEffects)))

(event EventOnPawnAndPartChanged
  (bind _returnvalue (CustomCharacter|GetCharacterPartType))
  (bind _returnvalue_2 (CustomCharacter|GetAssociatedPlayerPawn))
  (bind _returnvalue_3 (Customization|GetSkeletalMeshforPartType _returnvalue_2 _returnvalue :!class /Script/FortniteGame.FortPlayerPawn))
  (bind _returnvalue_4 (Utilities|Time|GetGameTimeinSeconds))
  (|CallParentFunction)
  (CallFunction|ApplyTransitionMaterial false)
  (Rendering|Material|SetCustomPrimitiveDataFloat _returnvalue_3 0 _returnvalue_4)
  ;[1 1 1 1 40] Backpack Material Transition IN
  (Rendering|Material|SetCustomPrimitiveDataFloat _returnvalue_3 1 (+ _returnvalue_4 999999.0))
  (bind _returnvalue_5 (PartModifier|SpawnSystem|SpawnSystemAttachedPartModifier "/BRCosmetics/Accessories/FORT_Capes/Cape_F_MED_Hightower_Squash/FX/NS_Cape_HighTower_Squash.NS_Cape_HighTower_Squash" (CustomCharacter|GetSkeletalMeshForPartOnPlayerPawn) "root" "0, 0, 0" "0, 0, 0" "KeepRelativeOffset" true false)
    (:Success
      (Effects|Components|ParticleSystem|SetFloatParameter _returnvalue_5 "Delay" 0.3)
      (Components|Activation|Activate _returnvalue_5 true))
    (:Failure
      _))
  (Utilities|FlowControl|Delay (Variables|Default|GetTransition%20In%20Time))
  (CallFunction|ApplyTransitionMaterial true))

(fn ApplyTransitionMaterial
  (params (RestoreOriginalMaterial :type Boolean))
  (bind _success (CallFunction|GetVariantLevel))
  (if _success
    (Rendering|Material|SetMaterial (CustomCharacter|GetSkeletalMeshForPartOnPlayerPawn) 0 (Utilities|Array|Get(acopy) (select RestoreOriginalMaterial (Variables|Default|GetSourceMaterials) (Variables|Default|GetTransitionMaterials)) (Math|Integer|Clamp(Integer) (Variables|Default|GetCurrentVariant) 0 (Utilities|Array|LastIndex (select RestoreOriginalMaterial (Variables|Default|GetSourceMaterials) (Variables|Default|GetTransitionMaterials))))))))

(fn GetCapeVisibility
  (return (Variables|/Script/Engine.SceneComponent|GetbVisible (CustomCharacter|GetSkeletalMeshForPartOnPlayerPawn))))

(fn GetVariantLevel
  (bind _swapdata (CosmeticLoadout|ConstructItemSwapClonedfromLoadout (CosmeticLoadout|GetCosmeticLoadout) (CosmeticLoadout|GetItemDefinitionfromLoadout (CosmeticLoadout|GetCosmeticLoadout) "(TagName=\"LoadoutSlot.Character\")")))
  (bind _array_element (Utilities|Array|ForEachLoopwithBreak (Utilities|Struct|BreakFortSwapItemandVariantData _swapdata))
    (:LoopBody
      (if (GameplayTags|Equal(GameplayTag) (.VariantChannelTag (Utilities|Struct|BreakMcpVariantChannelInfo _array_element)) "(TagName=\"Cosmetics.Variant.Channel.Progressive\")")
        (switch Utilities|FlowControl|Switch|SwitchonGameplayTag (.ActiveVariantTag (Utilities|Struct|BreakMcpVariantChannelInfo _array_element))
          (:Default
            _)
          (:Cosmetics.Variant.Property.Stage1
            _)
          (:Cosmetics.Variant.Property.Stage2
            (Variables|Default|SetCurrentVariant 1)
            (return true))
          (:Cosmetics.Variant.Property.Stage3
            (Variables|Default|SetCurrentVariant 2))
          (:Cosmetics.Variant.Property.Stage4
            (Variables|Default|SetCurrentVariant 3))
          (:Cosmetics.Variant.Property.Stage5
            (Variables|Default|SetCurrentVariant 4)))
        (Variables|Default|SetCurrentVariant 0)))
    (:Completed
      (Variables|Default|SetCurrentVariant 0)
      (return false))))

(fn ConstructionScript
  (|CallParentFunction))
