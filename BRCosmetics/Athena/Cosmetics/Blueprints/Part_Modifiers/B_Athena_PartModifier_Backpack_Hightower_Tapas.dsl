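; Backpack part modifier for the Hightower Tapas cosmetic that adds looping cape FX when worn with its matching character.
; After the post-res-in event, it compares the loadout's Character slot item to the Tapas character definition.
; On a match, it spawns the looping Niagara system attached at Root on the part's skeletal mesh.
; If the spawn succeeds, it applies the Character slot's variants to that component, so the FX follows the character's style.
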(event EventOnPostResIn
  (|CallParentFunction)
  (if (Utilities|Equal(Object) (CosmeticLoadout|GetItemDefinitionfromLoadout (CosmeticLoadout|GetCosmeticLoadout) "(TagName=\"LoadoutSlot.Character\")") (Variables|Default|GetCID%20Tapas))
    (bind _returnvalue (PartModifier|SpawnSystem|SpawnSystemAttachedPartModifier "/BRCosmetics/Accessories/FORT_Capes/Cape_M_MED_Tapas/FX/NS_Tapas_Backpack_LoopingFX_M.NS_Tapas_Backpack_LoopingFX_M" (CustomCharacter|GetSkeletalMeshForPartOnPlayerPawn) "Root" "0, 0, 0" "0, 0, 0" "SnapToTargetIncludingScale")
      (:Success
        (Class|FortCosmeticFunctionLibrary|ApplyVariantstoComponentbySlotTag (CustomCharacter|GetAssociatedPlayerPawn) "(TagName=\"LoadoutSlot.Character\")" _returnvalue)))))

(fn ConstructionScript
  (|CallParentFunction))
