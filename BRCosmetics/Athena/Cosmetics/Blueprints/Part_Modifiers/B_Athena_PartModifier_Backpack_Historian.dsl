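; Part modifier Blueprint for the Historian back bling cosmetic, toggling its idle VFX around a wink animation.
; The HistorianWink anim notify deactivates the component returned by GetIdleVFXComponent.
; The HistorianWink_End notify reactivates that component without a reset.
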(event Custom|AnimNotify_HistorianWink
  (Components|Activation|Deactivate (CallFunction|GetIdleVFXComponent)))

(event Custom|AnimNotify_HistorianWink_End
  (Components|Activation|Activate (CallFunction|GetIdleVFXComponent) false))

(fn ConstructionScript)
