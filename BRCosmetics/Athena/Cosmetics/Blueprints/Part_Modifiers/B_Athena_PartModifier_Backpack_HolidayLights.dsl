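; Backpack cosmetic part modifier for the Holiday Lights back bling that tells its material whether it is shown in the frontend.
; On the post-res-in event, after the parent logic, it gets the player pawn's skeletal mesh for the part and a dynamic material instance from it.
; It then sets that material's IsInFrontEnd scalar parameter to on or off based on the frontend check.
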(event EventOnPostResIn
  (|CallParentFunction)
  (Rendering|Material|SetScalarParameterValue (Material|GetMIDfromMaterial (CustomCharacter|GetSkeletalMeshForPartOnPlayerPawn) 1) "IsInFrontEnd" (Math|Float|SelectFloat 1.0 0.0 (CustomCosmeticModifier|IsinFrontEnd)) :!class /Script/Engine.MaterialInstanceDynamic))

(fn ConstructionScript
  (|CallParentFunction))
