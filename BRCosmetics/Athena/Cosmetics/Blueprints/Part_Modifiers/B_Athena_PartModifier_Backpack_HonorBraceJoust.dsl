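; Part modifier for the HonorBraceJoust back bling that reacts to the player's Athena kill count.
; Outside the front end, it maps kills to a stage through reversed threshold lookup and stores it, updating on cosmetic state activation.
; In matches, each watched kill change activates idle VFX and switches the Reactivity_Toggle material scalar on for three seconds.
; In frontend display, a timeline drives that scalar instead and actor tick is disabled.
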(event EventOnPawnAndPartChanged
  ; Stage Set
  (|CallParentFunction)
  (if (not (CustomCosmeticModifier|IsinFrontEnd))
    (bind _returnvalue (Class|FortCosmeticFunctionLibrary|GetCosmeticStatValue (CustomCharacter|GetAssociatedPlayerPawn) "(TagName=\"StatManager.AthenaKills\")"))
    (bind _stage (CallFunction|StatValuetoStage _returnvalue))
    (Variables|Default|SetCurrentStage _stage)))

(event EventOnPostResIn
  ; Event Timeline
  (|CallParentFunction)
  (if (Reactive|Frontend|IsCosmeticObjectDisplayedinFrontendContext 0)
    (Actor|Tick|SetActorTickEnabled false)
    (bind (_direction _kill) (Timeline|Timeline_2 PlayFromStart)
      (:Update
        (Rendering|Material|SetScalarParameterValueonMaterials (CustomCharacter|GetSkeletalMeshForPartOnPlayerPawn) "Reactivity_Toggle" _kill)))
    (else
      (CallFunction|PlayTAMaterial))))

(event Custom|Play%20TA%20Material
  ;[0.1418 0.05353 0.15 0.5] When killing an enemy is played animation
  (Cosmetics|WatchCosmeticStat (CustomCharacter|GetAssociatedPlayerPawn) "(TagName=\"StatManager.AthenaKills\")" 0 false
    (:OnStatChanged
      (Components|Activation|Activate (CallFunction|GetIdleVFXComponent) true)
      (Rendering|Material|SetScalarParameterValueonMaterials (CustomCharacter|GetSkeletalMeshForPartOnPlayerPawn) "Reactivity_Toggle" 1.0)
      (Utilities|FlowControl|Delay 3.0)
      (Rendering|Material|SetScalarParameterValueonMaterials (CustomCharacter|GetSkeletalMeshForPartOnPlayerPawn) "Reactivity_Toggle" 0.0))))

(event Custom|Stage%201
  ; Stages
  (Effects|Components|ParticleSystem|SetBoolParameter (CallFunction|GetIdleVFXComponent) "Stage_01" true))

(event EventOnCosmeticStateActivated
  (params (Stat%20Name :type Name :default "None") (Stat%20Value :type Integer) (STAGE%20Num%20Being%20Activated :type Integer))
  ; Update Mats
  (|CallParentFunction Stat%20Name Stat%20Value STAGE%20Num%20Being%20Activated)
  (Variables|Default|SetCurrentStage STAGE%20Num%20Being%20Activated))

(fn StatValueToStage
  (params (Stat%20Value :type Integer))
  (bind (_arrayindex _arrayelement) (Utilities|Array|ReverseforEachLoop (Variables|PartModifierControls|WatchCosmeticStat|GetStat%20Thresholds))
    (:LoopBody
      (if (>= (Variables|GetStat%20Value) _arrayelement)
        (return _arrayindex)))
    (:Completed
      (return 0))))

(fn ConstructionScript
  (|CallParentFunction))
