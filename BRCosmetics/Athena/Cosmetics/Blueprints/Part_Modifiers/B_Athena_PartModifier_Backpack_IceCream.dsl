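; Cosmetic part modifier for the Ice Cream back bling that makes the snowcone and its sparkle react to the wearer's health.
; After post-res-in it watches player health, scaling the idle VFX spawn rate and glint radius with the health percentage.
; A timeline lerps the mesh's "shrinkage" material parameter from old to current percentage, skipped while already playing.
; A GainedHealthTimes stat watch spawns a debounced fill-up emitter on FX_Socket.
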(event EventOnPostResIn
  (bind _returnvalue (CustomCharacter|GetAssociatedPlayerPawn))
  (bind _returnvalue_2 (CallFunction|GetIdleVFXComponent))
  (|CallParentFunction)
  ;[1 1 1] Watch player hit points
  (bind _returnvalue_3 (Fortnite|Utilities|AttemptGetPlayerStateforPawn _returnvalue))
  (bind _asfort_player_state_athena (Utilities|Casting|CastToFortPlayerStateAthena _returnvalue_3))
  (bind (_health _maxhealth) (Cosmetics|WatchPlayerHealth _asfort_player_state_athena)
    (:then
      ;[1 1 1] Play vomit effect
      (Cosmetics|WatchCosmeticStat _returnvalue "(TagName=\"StatManager.GainedHealthTimes\")" 0 false
        (:OnStatChanged
          (PartModifier|SpawnSystem|SpawnEmitterAttachedPartModifier(Legacy) :EmitterTemplate "/BRCosmetics/Accessories/FORT_Backpacks/M_MED_IceCream_Backpack/FX/P_Backpack_IceCream_Fillup.P_Backpack_IceCream_Fillup" :AttachToComponent (CustomCharacter|GetSkeletalMeshForPartOnPlayerPawn) :AttachPointName "FX_Socket" :Scale "1.000000,1.000000,1.000000" :LocationType "SnapToTargetIncludingScale" :bAutoDestroy true :PoolingMethod "None" :bAutoActivate true :bMatchPawnCustomDepth true :bDebounce true :DebounceTime -1.0))))
    (:OnHitPointsChanged
      (Utilities|FlowControl|Sequence
        (:then_0
          ;[1 1 1] Update percentage and effect scale
          (bind _output_get (Variables|Default|SetCurrentHealth _health))
          (bind _output_get_2 (Variables|Default|SetCurrentHealth_Percentage (Math|Float|SafeDivide _output_get _maxhealth)))
          (bind _output_get_3 (Variables|Default|SetGlintScaling (* _output_get_2 1e+01)))
          (Effects|Components|ParticleSystem|SetFloatParameter _returnvalue_2 "SpawnRate" _output_get_3)
          (Effects|Components|ParticleSystem|SetFloatParameter _returnvalue_2 "Glint_SpawnRadius" (/ _output_get_3 2.5)))
        (:then_1
          ;[1 1 1] Run snowcone growth timeline
          (if (and (not (Math|Float|NearlyEqual(Float) (Variables|Default|GetCurrentHealth_Percentage) (Variables|Default|GetOldHealth_Percentage))) (not (Components|Timeline|IsPlaying (Variables|B_Athena_PartModifier_Backpack_IceCream|GetTimeline_0))))
            (bind (_direction _newtrack_0) (Timeline|Timeline_0 PlayFromStart)
              (:Update
                (Rendering|Material|SetScalarParameterValueonMaterials (CustomCharacter|GetSkeletalMeshForPartOnPlayerPawn) "shrinkage" (- 1.0 (Math|Float|Lerp (Variables|Default|GetOldHealth_Percentage) (Variables|Default|GetCurrentHealth_Percentage) _newtrack_0))))
              (:Finished
                ;[1 1 1] Update old health with current health
                (Variables|Default|SetOldHealth (Variables|Default|GetCurrentHealth))
                (Variables|Default|SetOldHealth_Percentage (Variables|Default|GetCurrentHealth_Percentage))))))))))

(fn ConstructionScript)
