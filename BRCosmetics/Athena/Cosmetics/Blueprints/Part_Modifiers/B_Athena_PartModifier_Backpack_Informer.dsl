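; Backpack part modifier for the Informer cosmetic that makes the back bling glitch when the player takes damage.
; On post-resurrection init it watches the DamageReceived cosmetic stat on the associated player pawn.
; Each stat change restarts a timeline, whose reaction curve drives the "Glitch" scalar parameter on the materials of the part's skeletal mesh.
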(event EventOnPostResIn
  (|CallParentFunction)
  (Cosmetics|WatchCosmeticStat (CustomCharacter|GetAssociatedPlayerPawn) "(TagName=\"StatManager.DamageReceived\")"
    (:OnStatChanged
      (bind (_direction _reaction) (Timeline|Timeline_0 PlayFromStart)
        (:Update
          (Rendering|Material|SetScalarParameterValueonMaterials (CustomCharacter|GetSkeletalMeshForPartOnPlayerPawn) "Glitch" _reaction))))))

(fn ConstructionScript)
