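; Cosmetic part modifier for the JazzShoes backpack that changes its facial expression in reaction to player gameplay.
; After post-init outside the front end, it watches chest-open and damage-dealt stats, triggering three-second expressions through material scalar pose parameters that pause GPU blinking.
; An open-chest flag stops damage reactions from starting during a chest reaction.
; A DBNO state delegate, bound when DBNO starts, toggles a separate downed expression.
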(event EventPlayerStartDBNO
  (|CallParentFunction)
  (EventDispatchers|BindEventtoOnDBNOStateChanged (Variables|Default|GetPlayerPawn) (delegate Custom|DBNO%20Emote)))

(event EventOnPostResIn
  (bind _returnvalue (CustomCharacter|GetAssociatedPlayerPawn))
  (|CallParentFunction)
  (Variables|Default|SetMesh (CustomCharacter|GetSkeletalMeshForPartOnPlayerPawn))
  (Variables|Default|SetPlayerPawn _returnvalue)
  (if (not (CustomCosmeticModifier|IsinFrontEnd))
    (Utilities|FlowControl|Sequence
      (:then_0
        (Cosmetics|WatchCosmeticStat _returnvalue "(TagName=\"StatManager.OpenChests\")" 0 false
          (:OnStatChanged
            (CallFunction|OpenChestEmote))))
      (:then_1
        (Cosmetics|WatchCosmeticStat _returnvalue "(TagName=\"StatManager.DamageDealt\")" 0 false
          (:OnStatChanged
            (CallFunction|DamageDealtEmote)))))))

(event Custom|DBNO%20Emote
  (params (FortPawn :type Fort%20Pawn%20Object%20Reference) (bInIsDBNO :type Boolean))
  (bind _mesh (Variables|Default|GetMesh))
  (bind _mesh_2 (Variables|Default|GetMesh))
  (bind _mesh_3 (Variables|Default|GetMesh))
  (if bInIsDBNO
    (Rendering|Material|SetScalarParameterValueonMaterials _mesh_3 "UseGPUBlink" 0.0)
    (Rendering|Material|SetScalarParameterValueonMaterials _mesh_2 "R_Frown_pose" 1.0)
    (Rendering|Material|SetScalarParameterValueonMaterials _mesh_2 "C_glabella_up_pose" 1.0)
    (Rendering|Material|SetScalarParameterValueonMaterials _mesh_2 "jaw_open_pose" 1.0)
    (else
      (Rendering|Material|SetScalarParameterValueonMaterials _mesh_3 "UseGPUBlink" 1.0)
      (Rendering|Material|SetScalarParameterValueonMaterials _mesh "R_Frown_pose" 0.0)
      (Rendering|Material|SetScalarParameterValueonMaterials _mesh "C_glabella_up_pose" 0.0)
      (Rendering|Material|SetScalarParameterValueonMaterials _mesh "jaw_open_pose" 0.0))))

(event EventOnDayPhaseChanged
  (params (CurrentDayPhase :type EFortDayPhase%20Enum :default "Morning") (PreviousDayPhase :type EFortDayPhase%20Enum :default "Morning") (bAtCreation :type Boolean)))

(event Custom|OpenChestEmote
  (bind _mesh (Variables|Default|GetMesh))
  (bind _mesh_2 (Variables|Default|GetMesh))
  (Variables|Default|SetIsOpenChest true)
  (Rendering|Material|SetScalarParameterValueonMaterials _mesh "UseGPUBlink" 0.0)
  (Rendering|Material|SetScalarParameterValueonMaterials _mesh "jaw_open_pose" 1.0)
  (Rendering|Material|SetScalarParameterValueonMaterials _mesh "R_brow_up_pose" 1.0)
  (Utilities|FlowControl|RetriggerableDelay 3.0)
  (Rendering|Material|SetScalarParameterValueonMaterials _mesh_2 "UseGPUBlink" 1.0)
  (Rendering|Material|SetScalarParameterValueonMaterials _mesh_2 "jaw_open_pose" 0.0)
  (Rendering|Material|SetScalarParameterValueonMaterials _mesh_2 "R_brow_up_pose" 0.0)
  (Variables|Default|SetIsOpenChest false))

(event EventOnNewPlayerWeaponType
  (params (NewType :type EFortItemType%20Enum :default "WorldItem") (OldType :type EFortItemType%20Enum :default "WorldItem"))
  (|CallParentFunction NewType OldType) ; This node is disabled and will not be called.\nDrag off pins to build functionality.
)

(event EventOnWeaponChanged
  (params (NewWeapon :type Fort%20Weapon%20Object%20Reference) (OldWeapon :type Fort%20Weapon%20Object%20Reference)))

(event EventOnWeaponFired)

(event EventOnNewWeapons
  (params (WeaponsAsItems :type Array%20of%20Fort%20Item%20Object%20References)))

(event EventOnCurieStateAttached
  (params (Tag :type Gameplay%20Tag%20Structure%20%28by%20ref%29)))

(event Custom|DamageDealtEmote
  (bind _mesh (Variables|Default|GetMesh))
  (bind _mesh_2 (Variables|Default|GetMesh))
  (if (not (Variables|Default|GetIsOpenChest))
    (Rendering|Material|SetScalarParameterValueonMaterials _mesh_2 "UseGPUBlink" 0.0)
    (Rendering|Material|SetScalarParameterValueonMaterials _mesh_2 "R_Frown_pose" 1.0)
    (Rendering|Material|SetScalarParameterValueonMaterials _mesh_2 "R_smile_pose" 0.0)
    (Rendering|Material|SetScalarParameterValueonMaterials _mesh_2 "C_glabella_down_pose" 1.0)
    (Utilities|FlowControl|RetriggerableDelay 3.0)
    (Rendering|Material|SetScalarParameterValueonMaterials _mesh "UseGPUBlink" 1.0)
    (Rendering|Material|SetScalarParameterValueonMaterials _mesh "R_Frown_pose" 0.0)
    (Rendering|Material|SetScalarParameterValueonMaterials _mesh "R_smile_pose" 1.0)
    (Rendering|Material|SetScalarParameterValueonMaterials _mesh "C_glabella_down_pose" 0.0)))

(event EventOnCurieStateDetached
  (params (Tag :type Gameplay%20Tag%20Structure%20%28by%20ref%29)))

(event EventOnBeginSkydiving
  (|CallParentFunction) ; This node is disabled and will not be called.\nDrag off pins to build functionality.
)

(event EventOnEndSkydiving
  (|CallParentFunction) ; This node is disabled and will not be called.\nDrag off pins to build functionality.
)

(event EventOnEnterVehicle
  (params (VehicleEntered :type Fort%20Vehicle%20Interface%20Interface%20%28by%20ref%29))
  (|CallParentFunction VehicleEntered) ; This node is disabled and will not be called.\nDrag off pins to build functionality.
)

(event EventOnExitVehicle
  (|CallParentFunction) ; This node is disabled and will not be called.\nDrag off pins to build functionality.
)

(event EventOnTargetingOrFirstPersonCameraChanged)

(event EventOnHitPawn
  (params (SelfActor :type Actor%20Object%20Reference) (OtherActor :type Actor%20Object%20Reference) (NormalImpulse :type Vector :default "0, 0, 0") (Hit :type Hit%20Result%20Structure%20%28by%20ref%29)))

(fn ConstructionScript
  (|CallParentFunction))
