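; Cosmetic part modifier for the Jekyll backpack that drives timed fire and idle VFX on the player pawn's pack mesh.
; After res-in it binds to the pack AnimBP's active-looping FX dispatcher, then briefly plays the Jekyll fire component attached at a configurable socket.
; When the dispatcher fires, it restarts the idle VFX, then after a delay reattaches the fire and runs it for a fixed window.
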(event EventOnPostResIn
  (bind _jekyll_fire (Variables|Default|GetJekyll_Fire))
  (|CallParentFunction)
  (bind _asm_med_jekyll_pack_anim_bp (Utilities|Casting|CastToM_MED_Jekyll_Pack_AnimBP (Components|SkeletalMesh|GetAnimInstance (Customization|GetSkeletalMeshforPartType (CustomCharacter|GetAssociatedPlayerPawn) (CustomCharacter|GetCharacterPartType) :!class /Script/FortniteGame.FortPlayerPawn))))
  (Default|BindEventtoActiveLoopingFX _asm_med_jekyll_pack_anim_bp (EventDispatchers|CreateEvent :!event Custom|Active%20Looping%20FX))
  (Transformation|AttachComponentToComponent _jekyll_fire (CustomCharacter|GetSkeletalMeshForPartOnPlayerPawn) (Variables|Default|GetFire%20FX%20Socket%20Name) "KeepRelative" "KeepRelative" "KeepRelative" true)
  (Components|Activation|Activate _jekyll_fire false)
  (Utilities|FlowControl|Delay 3.0)
  (Components|Activation|Deactivate _jekyll_fire))

(event Custom|Active%20Looping%20FX
  (bind _jekyll_fire (Variables|Default|GetJekyll_Fire))
  (Components|Activation|Activate (CallFunction|GetIdleVFXComponent) true)
  (Utilities|FlowControl|Delay 15.0)
  (Transformation|AttachComponentToComponent _jekyll_fire (CustomCharacter|GetSkeletalMeshForPartOnPlayerPawn) (Variables|Default|GetFire%20FX%20Socket%20Name) "KeepRelative" "KeepRelative" "KeepRelative" true)
  (Components|Activation|Activate _jekyll_fire false)
  (Utilities|FlowControl|Delay 11.0)
  (Components|Activation|Deactivate _jekyll_fire))

(fn ConstructionScript
  (|CallParentFunction))
