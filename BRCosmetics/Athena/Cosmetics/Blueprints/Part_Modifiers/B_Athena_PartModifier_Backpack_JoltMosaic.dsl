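; Cosmetic part modifier Blueprint for the JoltMosaic backpack that toggles its idle VFX component in reaction to events.
; On post res-in it switches the VFX off, and in the front end it then plays one brief delayed burst, followed by a trailing delay that triggers nothing.
; Outside the front end, any stat change activates the same tagged VFX for a few seconds, then deactivates it, without checking the stat name or value.
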(event EventOnPostResIn
  (bind _returnvalue (CallFunction|GetVFXComponent "(TagName=\"CosmeticPart.VFX.Idle\")"))
  (|CallParentFunction)
  (if (CustomCosmeticModifier|IsinFrontEnd)
    ;[1 1 1] Front-End Reactivity
    (Components|Activation|Deactivate _returnvalue)
    (Utilities|FlowControl|Delay 2.0)
    (Components|Activation|Activate _returnvalue false)
    (Utilities|FlowControl|Delay 2.5)
    (Components|Activation|Deactivate _returnvalue)
    (Utilities|FlowControl|Delay 15.0)
    (else
      (Components|Activation|Deactivate _returnvalue))))

(event EventOnStatChanged
  (params (State%20Name :type Name :default "None") (Stat%20Value :type Integer))
  (bind _returnvalue (CallFunction|GetVFXComponent "(TagName=\"CosmeticPart.VFX.Idle\")"))
  (|CallParentFunction State%20Name Stat%20Value)
  (if (not (CustomCosmeticModifier|IsinFrontEnd))
    (Components|Activation|Activate _returnvalue false)
    (Utilities|FlowControl|Delay 3.0)
    (Components|Activation|Deactivate _returnvalue)))

(fn ConstructionScript
  (|CallParentFunction))
