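; Cosmetic part modifier for the Jonesy Wombat back bling that matches its shard VFX to the equipped Progressive variant.
; On post-resource init it creates a dynamic shard material and passes it to the idle Niagara component.
; It reads the backpack slot's Progressive variant index, switching on it to set the Niagara Stage parameter and material CrackColor.
; An invalid index or missing channel prints a debug string.
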(event EventOnPostResIn
  (bind _returnvalue (CallFunction|GetIdleVFXComponent))
  (bind _shards_mid (Variables|Default|GetShards_MID))
  (|CallParentFunction)
  (bind _returnvalue_2 (Rendering|Material|CreateDynamicMaterialInstance "/BRCosmetics/Accessories/FORT_Backpacks/Backpack_M_MED_Jonesy_Wombat/VFX/MI_JonesyBackpack_Shards_Frag_Bit.MI_JonesyBackpack_Shards_Frag_Bit" "Shards_MiD" "None" :!class /Script/Engine.KismetMaterialLibrary))
  (Variables|Default|SetShards_MID _returnvalue_2)
  (Niagara|SetNiagaraVariable(Material) (Utilities|Casting|CastToNiagaraComponent (CallFunction|GetIdleVFXComponent)) "User.Shards_MID" (Variables|Default|GetShards_MID))
  (bind (_index _bindexisvalid _bfoundchannel) (CosmeticLoadout|GetVariantChannelActiveIndexforSlotActor "(TagName=\"LoadoutSlot.Backpack\")" (CustomCharacter|GetAssociatedPlayerPawn) "(TagName=\"Cosmetics.Variant.Channel.Progressive\")" (Variables|Default|GetVariant%20List)))
  (if (and _bindexisvalid _bfoundchannel)
    (switch int _index
      (:0
        _)
      (:1
        _)
      (:2
        (Effects|Components|ParticleSystem|SetIntParameter _returnvalue "User.Stage" 1)
        (Rendering|Material|SetVectorParameterValue _shards_mid "CrackColor" "(R=0.083333,G=0.000000,B=1.000000,A=1.000000)" :!class /Script/Engine.MaterialInstanceDynamic))
      (:3
        (Effects|Components|ParticleSystem|SetIntParameter _returnvalue "User.Stage" 2)
        (Rendering|Material|SetVectorParameterValue _shards_mid "CrackColor" "(R=1.000000,G=0.000000,B=1.000000,A=1.000000)" :!class /Script/Engine.MaterialInstanceDynamic))
      (:4
        (Effects|Components|ParticleSystem|SetIntParameter _returnvalue "User.Stage" 3)
        (Rendering|Material|SetVectorParameterValue _shards_mid "CrackColor" "(R=1.000000,G=0.000000,B=0.166667,A=1.000000)" :!class /Script/Engine.MaterialInstanceDynamic)))
    (Effects|Components|ParticleSystem|SetIntParameter _returnvalue "User.Stage" 0)
    (Rendering|Material|SetVectorParameterValue _shards_mid "CrackColor" "(R=0.000000,G=1.000000,B=2.000000,A=1.000000)" :!class /Script/Engine.MaterialInstanceDynamic)
    (else
      (Development|PrintString "Check_CCPM_JonesyBackpack (invalid index or no channel found)" true true "(R=1.000000,G=0.000000,B=0.014042,A=1.000000)"))))

(fn ConstructionScript
  (|CallParentFunction))
