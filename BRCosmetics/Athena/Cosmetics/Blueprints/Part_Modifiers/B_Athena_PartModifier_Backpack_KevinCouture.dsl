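; Part modifier for the Kevin Couture back bling that makes it react to Athena kills through staged glow.
; In the frontend it previews stages 0 to 3 on timed delays, firing anim blueprint reactivity and Niagara bursts between them.
; In matches it applies a stage from the clamped kill stat and on cosmetic state activation.
; Stages set emissive material parameters, with the idle VFX flag toggled at the first and last.
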(event EventOnPostResIn
  (|CallParentFunction)
  (if (CustomCosmeticModifier|IsinFrontEnd)
    ;[1 0.07549 0.9459] Get Backpack AnimBP To Trigger Reactivity in Frontend
    (bind _asm_med_kevin_couture_pack_anim_bp (Utilities|Casting|CastToM_MED_Kevin_Couture_Pack_AnimBP (Components|SkeletalMesh|GetAnimInstance (Customization|GetSkeletalMeshforPartType (CustomCharacter|GetAssociatedPlayerPawn) (CustomCharacter|GetCharacterPartType) :!class /Script/FortniteGame.FortPlayerPawn))))
    (bind _output_get (Variables|Default|SetAnimBP _asm_med_kevin_couture_pack_anim_bp))
    (Class|MMEDKevinCouturePackAnimBP|FireFrontendReactivity _output_get 0.0)
    (CallFunction|Stage0)
    (Utilities|FlowControl|Delay 4.0)
    (CallFunction|Burst)
    (Class|MMEDKevinCouturePackAnimBP|FireFrontendReactivity (Variables|Default|GetAnimBP) 1.0)
    (CallFunction|Stage1)
    (Utilities|FlowControl|Delay 4.0)
    (CallFunction|Burst)
    (Class|MMEDKevinCouturePackAnimBP|FireFrontendReactivity (Variables|Default|GetAnimBP) 2.0)
    (CallFunction|Stage2)
    (Utilities|FlowControl|Delay 4.0)
    (CallFunction|Burst)
    (Class|MMEDKevinCouturePackAnimBP|FireFrontendReactivity (Variables|Default|GetAnimBP) 3.0)
    (CallFunction|Stage3)
    (Utilities|FlowControl|Delay 4.0)
    (else
      (bind _returnvalue (Class|FortCosmeticFunctionLibrary|GetCosmeticStatValue (CustomCharacter|GetAssociatedPlayerPawn) "(TagName=\"StatManager.AthenaKills\")"))
      (switch int (Math|Integer|Clamp(Integer) _returnvalue 0 3)
        (:1
          (CallFunction|Stage1))
        (:2
          (CallFunction|Stage2))
        (:3
          (CallFunction|Stage3))))))

(event Custom|Stage%200
  (bind _returnvalue (CustomCharacter|GetSkeletalMeshForPartOnPlayerPawn))
  (Rendering|Material|SetScalarParameterValueonMaterials _returnvalue "Emissive_BrightnessMax" 0.0)
  (Rendering|Material|SetScalarParameterValueonMaterials _returnvalue "Emissive_BrightnessMin" 0.0)
  (Rendering|Material|SetScalarParameterValueonMaterials _returnvalue "Emissive mult" 0.08)
  (Effects|Components|ParticleSystem|SetBoolParameter (CallFunction|GetIdleVFXComponent) "IdleStage" false))

(event EventOnPawnAndPartChanged
  (|CallParentFunction))

(event Custom|Stage%201
  (bind _returnvalue (CustomCharacter|GetSkeletalMeshForPartOnPlayerPawn))
  (Rendering|Material|SetScalarParameterValueonMaterials _returnvalue "Emissive_BrightnessMax" 0.25)
  (Rendering|Material|SetScalarParameterValueonMaterials _returnvalue "Emissive_BrightnessMin" 0.0)
  (Rendering|Material|SetScalarParameterValueonMaterials _returnvalue "Emissive mult" 0.35))

(event EventOnCosmeticStateActivated
  (params (Stat%20Name :type Name :default "None") (Stat%20Value :type Integer) (STAGE%20Num%20Being%20Activated :type Integer))
  (|CallParentFunction Stat%20Name Stat%20Value STAGE%20Num%20Being%20Activated)
  (if (> STAGE%20Num%20Being%20Activated 0)
    (Cosmetics|WatchCosmeticStat (CustomCharacter|GetAssociatedPlayerPawn) "(TagName=\"StatManager.AthenaKills\")" 0 false
      (:OnStatChanged
        _))
    (else
      _))
  (switch int STAGE%20Num%20Being%20Activated
    (:1
      (CallFunction|Stage1))
    (:2
      (CallFunction|Stage2))
    (:3
      (CallFunction|Stage3))))

(event Custom|Stage%202
  (bind _returnvalue (CustomCharacter|GetSkeletalMeshForPartOnPlayerPawn))
  (Rendering|Material|SetScalarParameterValueonMaterials _returnvalue "Emissive_BrightnessMax" 4.0)
  (Rendering|Material|SetScalarParameterValueonMaterials _returnvalue "Emissive_BrightnessMin" 0.0)
  (Rendering|Material|SetScalarParameterValueonMaterials _returnvalue "Emissive  mult" 0.5))

(event Custom|Stage%203
  (bind _returnvalue (CustomCharacter|GetSkeletalMeshForPartOnPlayerPawn))
  (Rendering|Material|SetScalarParameterValueonMaterials _returnvalue "Emissive_BrightnessMax" 8.0)
  (Rendering|Material|SetScalarParameterValueonMaterials _returnvalue "Emissive_BrightnessMin" 0.1)
  (Rendering|Material|SetScalarParameterValueonMaterials _returnvalue "Emissive mult" 1.0)
  (Effects|Components|ParticleSystem|SetBoolParameter (CallFunction|GetIdleVFXComponent) "IdleStage" true))

(event Custom|Burst
  (PartModifier|SpawnSystem|SpawnSystemAttachedPartModifier "/BRCosmetics/Accessories/FORT_Backpacks/Backpack_M_MED_Kevin_Couture/FX/NS_Backpack_KevinCouture_Burst.NS_Backpack_KevinCouture_Burst" (CustomCharacter|GetSkeletalMeshForPartOnPlayerPawn) "root"))

(fn ConstructionScript
  (|CallParentFunction))
