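; Backpack part modifier for the KneeLens cosmetic that periodically triggers a timed material reaction on the part's skeletal mesh.
; After a delay, a self-rearming timer repeats the reaction, using fixed timing in the front end and randomized intervals elsewhere.
; Each activation writes game-time timestamp, duration and a FlipFlop-alternated frame selection into custom primitive data for the material.
; The mesh reference is cached on first lookup, and the timer is cleared on destroy.
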(event EventOnPostResIn
  (|CallParentFunction)
  (Variables|Default|SetIsFrontEnd (CustomCosmeticModifier|IsinFrontEnd))
  (Utilities|FlowControl|Delay (select (Variables|Default|GetIsFrontEnd) 3.0 (Math|Random|RandomFloatinRange 3.0 15.0)))
  (CallFunction|StartReactivityLoop))

(event Custom|StartReactivityLoop
  (CallFunction|ActivateReactivity)
  (bind _returnvalue (Utilities|Time|SetTimerbyEvent (delegate Custom|StartReactivityLoop) (+ (select (Variables|Default|GetIsFrontEnd) 12.0 (Math|Random|RandomFloatinRange 15.0 3e+01)) (Variables|Default|GetCurrentDuration)) false false 5.0))
  (Variables|Default|SetReactivityTimerHandle _returnvalue))

(event Custom|ActivateReactivity
  (Rendering|Material|SetCustomPrimitiveDataFloat (CallFunction|GetMesh) 1 (Utilities|Time|GetGameTimeinSeconds)) ; Timestamping
  (bind _output_get (Variables|Default|SetCurrentDuration (select (Variables|Default|GetIsFrontEnd) 5.0 7.0)))
  (Rendering|Material|SetCustomPrimitiveDataFloat (CallFunction|GetMesh) 2 _output_get) ; Reactivity Duration
  (bind _isa (Utilities|FlowControl|FlipFlop)
    (:A
      _)
    (:B
      _))
  (Rendering|Material|SetCustomPrimitiveDataFloat (CallFunction|GetMesh) 3 (Math|Conversions|ToFloat(Boolean) _isa)) ; Frame Selection
)

(event EventDestroyed
  (|CallParentFunction)
  (Utilities|Time|ClearandInvalidateTimerbyHandle (Variables|Default|GetReactivityTimerHandle)))

(fn GetMesh
  (Utilities|IsValid (Variables|Default|GetSkelMesh)
    (:Is%20Valid
      (return (Variables|Default|GetSkelMesh)))
    (:Is%20Not%20Valid
      (bind _output_get (Variables|Default|SetSkelMesh (CustomCharacter|GetSkeletalMeshForPartOnPlayerPawn)))
      (return _output_get))))

(fn ConstructionScript
  (|CallParentFunction))
