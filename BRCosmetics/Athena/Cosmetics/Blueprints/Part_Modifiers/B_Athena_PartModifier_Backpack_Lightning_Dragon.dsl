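; Cosmetic part modifier for the Lightning Dragon backpack that plays its Niagara lightning-dragon effect on the "pack" attachment.
; In eligible frontend display, excluding end holograms, post-res-in spawns the effect and starts a looping timer to replay it, cleared on destruction.
; A positive stat change also spawns it, gated by DoOnce with no visible reset.
; Emote start and end toggle a "Begin Emote" bool on the idle VFX component.
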(event Custom|CE_Frontend%20FX%20Loop%20Burst
  (PartModifier|SpawnSystem|SpawnSystemAttachedPartModifier "/BRCosmetics/Accessories/FORT_Backpacks/Backpack_F_MED_LightningDragon/FX/NS_Backpack_LightningDragon.NS_Backpack_LightningDragon" (CustomCharacter|GetSkeletalMeshForPartOnPlayerPawn) "pack"))

(event EventOnPostResIn
  (|CallParentFunction)
  (if (not (Variables|/Script/FortniteGame.FortPlayerPawn|GetbIsInFrontEndHologram (CustomCharacter|GetAssociatedPlayerPawn)))
    (if (Reactive|Frontend|ShouldDisplayFrontendReactivity 0 "Backpack")
      (Utilities|FlowControl|DoOnce false
        (:Completed
          (bind _returnvalue (Utilities|Time|SetTimerbyEvent (delegate Custom|CE_Frontend%20FX%20Loop%20Burst) 6.0 true))
          (Variables|Default|SetTimer _returnvalue)
          (PartModifier|SpawnSystem|SpawnSystemAttachedPartModifier "/BRCosmetics/Accessories/FORT_Backpacks/Backpack_F_MED_LightningDragon/FX/NS_Backpack_LightningDragon.NS_Backpack_LightningDragon" (CustomCharacter|GetSkeletalMeshForPartOnPlayerPawn) "pack"))))))

(event EventOnStatChanged
  (params (State%20Name :type Name :default "None") (Stat%20Value :type Integer))
  (|CallParentFunction State%20Name Stat%20Value)
  (if (> Stat%20Value 0)
    (Utilities|FlowControl|DoOnce false
      (:Completed
        (PartModifier|SpawnSystem|SpawnSystemAttachedPartModifier "/BRCosmetics/Accessories/FORT_Backpacks/Backpack_F_MED_LightningDragon/FX/NS_Backpack_LightningDragon.NS_Backpack_LightningDragon" (CustomCharacter|GetSkeletalMeshForPartOnPlayerPawn) "pack"
          (:Success
            (Utilities|FlowControl|RetriggerableDelay 1.7)))))))

(event EventOnBeginEmote
  (params (MontageItemDef :type Fort%20Item%20Definition%20Object%20Reference) (PawnEmoting :type Fort%20Pawn%20Object%20Reference))
  (|CallParentFunction MontageItemDef PawnEmoting)
  (bind _asniagara_particle_system_component (Utilities|Casting|CastToNiagaraComponent (CallFunction|GetIdleVFXComponent)))
  (Niagara|SetNiagaraVariable(Bool) _asniagara_particle_system_component "Begin Emote" true))

(event EventDestroyed
  (|CallParentFunction)
  (if (Reactive|Frontend|ShouldDisplayFrontendReactivity 0 "Backpack")
    (Utilities|Time|ClearandInvalidateTimerbyHandle (Variables|Default|GetTimer))))

(event EventOnEndEmote
  (params (MontageItemDef :type Fort%20Item%20Definition%20Object%20Reference) (PawnEmoting :type Fort%20Pawn%20Object%20Reference))
  (|CallParentFunction MontageItemDef PawnEmoting)
  (bind _asniagara_particle_system_component (Utilities|Casting|CastToNiagaraComponent (CallFunction|GetIdleVFXComponent)))
  (Niagara|SetNiagaraVariable(Bool) _asniagara_particle_system_component "Begin Emote" false))

(fn ConstructionScript
  (|CallParentFunction))
