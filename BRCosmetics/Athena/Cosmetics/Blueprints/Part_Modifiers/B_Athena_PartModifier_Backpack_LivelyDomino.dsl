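; Backpack cosmetic part modifier for LivelyDomino, wired to react to the wearer's kills.
; After res-in, it watches the associated player pawn's AthenaKills stat and calls ActiveKill when the stat changes.
; It then sets ReactivityEnabled from the frontend reactivity display check for the backpack slot.
; ActiveKill tests that flag but performs no action in its branch, so the visible reaction is not defined here.
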(event Custom|Active%20Kill
  (if (Variables|Default|GetReactivityEnabled)
    _))

(event EventOnPostResIn
  (|CallParentFunction)
  (Utilities|FlowControl|Sequence
    (:then_0
      (Cosmetics|WatchCosmeticStat (CustomCharacter|GetAssociatedPlayerPawn) "(TagName=\"StatManager.AthenaKills\")" 0 false
        (:OnStatChanged
          (CallFunction|ActiveKill))))
    (:then_1
      (Variables|Default|SetReactivityEnabled (Reactive|Frontend|ShouldDisplayFrontendReactivity 0 "Backpack")))))

(fn ConstructionScript
  (|CallParentFunction))
