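; Part modifier for the dark Llama Hero backpack cosmetic that plays particle effects as the wearer gets kills.
; When the pawn and part change, it watches the AthenaKills cosmetic stat on the associated player pawn and stores each new value as KillCount.
; Any count above zero spawns a kill emitter attached to the root of the part's skeletal mesh.
; A count of exactly one also spawns a separate start emitter there.
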(event EventOnPawnAndPartChanged
  (|CallParentFunction)
  (bind (_statname _statvalue) (Cosmetics|WatchCosmeticStat (CustomCharacter|GetAssociatedPlayerPawn) "(TagName=\"StatManager.AthenaKills\")" 0 false)
    (:OnStatChanged
      (bind _output_get (Variables|Default|SetKillCount (Math|Conversions|ToFloat(Integer) _statvalue)))
      (if (> _output_get 0.0)
        (PartModifier|SpawnSystem|SpawnEmitterAttachedPartModifier(Legacy) "/BRCosmetics/Characters/Player/Male/Medium/Bodies/M_MED_Llama_Hero/Skins/Dark/FX/P_Backpack_Llama_Hero_Dark_Kill.P_Backpack_Llama_Hero_Dark_Kill" (CustomCharacter|GetSkeletalMeshForPartOnPlayerPawn) "root"
          (:Success
            _)
          (:Failure
            _))
        (if (== (Variables|Default|GetKillCount) 1.0)
          (PartModifier|SpawnSystem|SpawnEmitterAttachedPartModifier(Legacy) "/BRCosmetics/Characters/Player/Male/Medium/Bodies/M_MED_Llama_Hero/Skins/Dark/FX/P_Backpack_Llama_Hero_Dark_Kill_Start.P_Backpack_Llama_Hero_Dark_Kill_Start" (CustomCharacter|GetSkeletalMeshForPartOnPlayerPawn) "root"))))))

(fn ConstructionScript)
