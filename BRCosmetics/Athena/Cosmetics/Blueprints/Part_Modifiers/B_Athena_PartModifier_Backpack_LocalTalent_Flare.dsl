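; Part modifier Blueprint for the LocalTalent Flare backpack cosmetic that makes its emissive react to stat changes, apparently eliminations.
; When a stat changes to a value above zero, a timeline plays from the start.
; Its animation curve lerps between the min and max reactivity brightness variables, driving the "EliminationActive" scalar parameter on the part's skeletal mesh materials.
; Post-res-in, cosmetic-state activation and construction only call their parent implementations.
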(event EventOnPostResIn
  (|CallParentFunction))

(event EventOnStatChanged
  (params (State%20Name :type Name :default "None") (Stat%20Value :type Integer))
  (|CallParentFunction State%20Name Stat%20Value)
  (if (> Stat%20Value 0)
    ;[1 0.06038 0.2724 0.5] Tech Art Reactivity
    (bind (_direction _brightness _animationcurve) (Timeline|Timeline PlayFromStart)
      (:Update
        (Rendering|Material|SetScalarParameterValueonMaterials (CustomCharacter|GetSkeletalMeshForPartOnPlayerPawn) "EliminationActive" (Math|Float|Lerp (Variables|Default|GetReactivity%20Emissive%20Brightness%20Min) (Variables|Default|GetReactivity%20Emissive%20Brightness%20Max) _animationcurve))))))

(event EventOnCosmeticStateActivated
  (params (Stat%20Name :type Name :default "None") (Stat%20Value :type Integer) (STAGE%20Num%20Being%20Activated :type Integer))
  (|CallParentFunction Stat%20Name Stat%20Value STAGE%20Num%20Being%20Activated))

(fn ConstructionScript
  (|CallParentFunction))
