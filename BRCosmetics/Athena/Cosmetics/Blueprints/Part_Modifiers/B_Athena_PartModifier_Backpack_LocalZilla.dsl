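; Cosmetic part modifier for the LocalZilla backpack that makes its emissive glow react to damage and kills.
; After res-in, it creates the backpack MID reference, caches its base emissive, and sets fixed glow levels, dimming the idle VFX in the front end.
; Outside the front end, cosmetic stat watchers play Flickering on damage and Boost on kills, scaling material and VFX emissive.
; Flickering plays slower on Switch and mobile.
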(event EventOnPostResIn
  (bind _returnvalue (CallFunction|GetIdleVFXComponent))
  (bind _initialemissiveintensity (Variables|Default|GetInitialEmissiveIntensity))
  (bind _boostfactor (Variables|Default|GetBoostFactor))
  (bind _initialemissiveintensity_2 (Variables|Default|GetInitialEmissiveIntensity))
  (|CallParentFunction)
  (Networking|CanExecuteCosmeticEvents
    (:True
      (Components|Timeline|SetPlayRate (Variables|B_Athena_PartModifier_Backpack_LocalZilla|GetFlickering) (Math|Float|SelectFloat 0.6 0.8 (or (Utilities|String|Equal,CaseInsensitive(String) (Game|GetPlatformName) "Switch") (Fortnite|IsMobilePlatform))))
      (bind _asmaterial_instance_dynamic (Utilities|Casting|CastToMaterialInstanceDynamic (Rendering|Material|GetMaterial (CustomCharacter|GetSkeletalMeshForPartOnPlayerPawn) 0)))
      (bind _output_get (Variables|Default|SetBackpackMID _asmaterial_instance_dynamic))
      (Utilities|IsValid _output_get
        (:Is%20Valid
          (bind _returnvalue_2 (Rendering|Material|GetScalarParameterValue _output_get "emissive mult" :!class /Script/Engine.MaterialInstanceDynamic))
          (Variables|Default|SetInitialEmissiveIntensity _returnvalue_2)
          (Variables|Default|SetAssociatedPlayerPawn (CustomCharacter|GetAssociatedPlayerPawn))
          (if (CustomCosmeticModifier|IsinFrontEnd)
            (Effects|Components|ParticleSystem|SetFloatParameter _returnvalue "Emissive" 0.3)
            (Rendering|Material|SetScalarParameterValue (Variables|Default|GetBackpackMID) "emissive mult" 2.0 :!class /Script/Engine.MaterialInstanceDynamic)
            (else
              (Effects|Components|ParticleSystem|SetFloatParameter _returnvalue "Emissive" 1.0)
              (Utilities|FlowControl|Sequence
                (:then_0
                  (bind (_statname _statvalue) (Cosmetics|WatchCosmeticStat (Variables|Default|GetAssociatedPlayerPawn) "(TagName=\"StatManager.DamageReceived\")" -1 false)
                    (:OnStatChanged
                      (bind (_direction _flicker) (Timeline|Flickering PlayFromStart)
                        (:Update
                          (Rendering|Material|SetScalarParameterValue (Variables|Default|GetBackpackMID) "emissive mult" (* _flicker _initialemissiveintensity_2) :!class /Script/Engine.MaterialInstanceDynamic)
                          (Effects|Components|ParticleSystem|SetFloatParameter (CallFunction|GetIdleVFXComponent) "Emissive" (* _flicker _initialemissiveintensity_2)))))))
                (:then_1
                  (Cosmetics|WatchCosmeticStat (Variables|Default|GetAssociatedPlayerPawn) "(TagName=\"StatManager.AthenaKills\")" -1 false
                    (:OnStatChanged
                      (bind (_direction_2 _boost) (Timeline|Boost PlayFromStart)
                        (:Update
                          (Rendering|Material|SetScalarParameterValue (Variables|Default|GetBackpackMID) "emissive mult" (* (+ (* _boost _boostfactor) 1.0) _initialemissiveintensity) :!class /Script/Engine.MaterialInstanceDynamic)
                          (Effects|Components|ParticleSystem|SetFloatParameter (CallFunction|GetIdleVFXComponent) "Emissive" (* (+ (* _boost _boostfactor) 1.0) _initialemissiveintensity)))))))))))))))

(event EventOnDayPhaseChanged
  (params (CurrentDayPhase :type EFortDayPhase%20Enum :default "Morning") (PreviousDayPhase :type EFortDayPhase%20Enum :default "Morning") (bAtCreation :type Boolean)))

(event EventOnNewPlayerWeaponType
  (params (NewType :type EFortItemType%20Enum :default "WorldItem") (OldType :type EFortItemType%20Enum :default "WorldItem"))
  (|CallParentFunction NewType OldType) ; This node is disabled and will not be called.\nDrag off pins to build functionality.
)

(event EventOnWeaponChanged
  (params (NewWeapon :type Fort%20Weapon%20Object%20Reference) (OldWeapon :type Fort%20Weapon%20Object%20Reference)))

(event EventOnWeaponFired)

(event EventOnNewWeapons
  (params (WeaponsAsItems :type Array%20of%20Fort%20Item%20Object%20References)))

(event EventOnCurieStateAttached
  (params (Tag :type Gameplay%20Tag%20Structure%20%28by%20ref%29)))

(event EventOnCurieStateDetached
  (params (Tag :type Gameplay%20Tag%20Structure%20%28by%20ref%29)))

(event EventOnBeginSkydiving
  (|CallParentFunction) ; This node is disabled and will not be called.\nDrag off pins to build functionality.
)

(event EventOnEndSkydiving
  (|CallParentFunction) ; This node is disabled and will not be called.\nDrag off pins to build functionality.
)

(event EventOnEnterVehicle
  (params (VehicleEntered :type Fort%20Vehicle%20Interface%20Interface%20%28by%20ref%29))
  (|CallParentFunction VehicleEntered) ; This node is disabled and will not be called.\nDrag off pins to build functionality.
)

(event EventOnExitVehicle
  (|CallParentFunction) ; This node is disabled and will not be called.\nDrag off pins to build functionality.
)

(event EventOnTargetingOrFirstPersonCameraChanged)

(event EventOnHitPawn
  (params (SelfActor :type Actor%20Object%20Reference) (OtherActor :type Actor%20Object%20Reference) (NormalImpulse :type Vector :default "0, 0, 0") (Hit :type Hit%20Result%20Structure%20%28by%20ref%29)))

(fn ConstructionScript
  (|CallParentFunction))
