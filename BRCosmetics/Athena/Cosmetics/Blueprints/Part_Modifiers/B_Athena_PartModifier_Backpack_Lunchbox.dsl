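; Part modifier Blueprint for the Lunchbox backpack cosmetic that plays a lightning and glow effect.
; When a tracked stat changes to a positive value, or on post-res-in in the front end, it waits briefly, then spawns an attached Niagara lightning system on the part's skeletal mesh.
; A timeline then drives the mesh material's "emissive mult" parameter, and still plays if the spawn fails, after printing a debug message.
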(event EventOnStatChanged
  (params (State%20Name :type Name :default "None") (Stat%20Value :type Integer))
  (|CallParentFunction State%20Name Stat%20Value)
  (if (> Stat%20Value 0)
    (Utilities|FlowControl|Delay 0.2)
    (PartModifier|SpawnSystem|SpawnSystemAttachedPartModifier "/BRCosmetics/Accessories/FORT_Backpacks/Backpack_M_MED_Lunch_Box/FX/NS_Lunchbox_Backpack_Lightning.NS_Lunchbox_Backpack_Lightning" (CustomCharacter|GetSkeletalMeshForPartOnPlayerPawn) "root" "0, 0, 0" "0, 0, 0" "KeepRelativeOffset" true true true "None" false
      (:Success
        (bind (_direction _newtrack_0) (Timeline|Timeline_0 PlayFromStart)
          (:Update
            (Rendering|Material|SetScalarParameterValueonMaterials (CustomCharacter|GetSkeletalMeshForPartOnPlayerPawn) "emissive mult" _newtrack_0))))
      (:Failure
        (Development|PrintString "Spawn Failure")
        (Timeline|Timeline_0 PlayFromStart)))))

(event EventOnPostResIn
  (|CallParentFunction)
  (if (CustomCosmeticModifier|IsinFrontEnd)
    (Utilities|FlowControl|Delay 0.2)
    (PartModifier|SpawnSystem|SpawnSystemAttachedPartModifier "/BRCosmetics/Accessories/FORT_Backpacks/Backpack_M_MED_Lunch_Box/FX/NS_Lunchbox_Backpack_Lightning.NS_Lunchbox_Backpack_Lightning" (CustomCharacter|GetSkeletalMeshForPartOnPlayerPawn) "root" "0, 0, 0" "0, 0, 0" "KeepRelativeOffset" true true true "None" false
      (:Success
        (Timeline|Timeline_0 PlayFromStart))
      (:Failure
        (Development|PrintString "Spawn Failure")
        (Timeline|Timeline_0 PlayFromStart)))))

(fn ConstructionScript
  (|CallParentFunction))
