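; Cosmetic part modifier Blueprint for the Snowman Fashion backpack that adds jump and land effects.
; When the pawn jumps or lands, it spawns the matching jump or land Niagara system attached to the backpack's skeletal mesh.
; Both effects attach at the socket returned by GetIdleFXSocketName; the construction script is empty.
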(event EventOnPawnJumped
  (PartModifier|SpawnSystem|SpawnSystemAttachedPartModifier "/BRCosmetics/Accessories/FORT_Backpacks/M_MED_Snowman_Fashion_Backpack/FX/NS_Backpack_SnowmanFashion_Jump.NS_Backpack_SnowmanFashion_Jump" (CustomCharacter|GetSkeletalMeshForPartOnPlayerPawn) (CallFunction|GetIdleFXSocketName)))

(event EventOnPawnLanded
  (PartModifier|SpawnSystem|SpawnSystemAttachedPartModifier "/BRCosmetics/Accessories/FORT_Backpacks/M_MED_Snowman_Fashion_Backpack/FX/NS_Backpack_SnowmanFashion_Land.NS_Backpack_SnowmanFashion_Land" (CustomCharacter|GetSkeletalMeshForPartOnPlayerPawn) (CallFunction|GetIdleFXSocketName)))

(fn ConstructionScript)
