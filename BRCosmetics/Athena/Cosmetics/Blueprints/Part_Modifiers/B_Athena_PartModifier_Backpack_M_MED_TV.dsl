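; Cosmetic part modifier for the Athena "M_MED_TV" back bling that glitches the backpack when its wearer takes damage.
; After post-resource init, it watches the StatManager.DamageReceived cosmetic stat on the associated player pawn.
; Each stat change with a positive value restarts a timeline that drives the BP_GlitchesIntensity scalar on the part's skeletal mesh materials.
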(event EventOnPostResIn
  (|CallParentFunction)
  (bind (_statname _statvalue) (Cosmetics|WatchCosmeticStat (CustomCharacter|GetAssociatedPlayerPawn) "(TagName=\"StatManager.DamageReceived\")")
    (:OnStatChanged
      (if (> (Math|Conversions|ToFloat(Integer) _statvalue) 0.0)
        (bind (_direction _staticintensity) (Timeline|Timeline_0 PlayFromStart)
          (:Update
            (Rendering|Material|SetScalarParameterValueonMaterials (CustomCharacter|GetSkeletalMeshForPartOnPlayerPawn) "BP_GlitchesIntensity" _staticintensity)))))))

(fn ConstructionScript
  (|CallParentFunction))
