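; Part modifier for the MagmaBreak backpack that makes it react visually to the player's eliminations.
; After res-in, it either plays a delayed preview reaction when frontend reactivity should display, or watches the associated pawn's AthenaKills stat and reacts to changes above zero.
; Each reaction stamps game time into the mesh materials' Timestamp parameter and passes an incremented Eliminations counter to the idle VFX.
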(event EventOnPostResIn
  (|CallParentFunction)
  (if (Reactive|Frontend|ShouldDisplayFrontendReactivity 0 "Backpack")
    (Utilities|FlowControl|Delay 2.0)
    (CallFunction|ReactivityFX)
    (Utilities|FlowControl|Delay 12.0)
    (else
      (bind (_statname _statvalue) (Cosmetics|WatchCosmeticStat (CustomCharacter|GetAssociatedPlayerPawn) "(TagName=\"StatManager.AthenaKills\")" -1 false)
        (:OnStatChanged
          (if (> _statvalue 0)
            (CallFunction|ReactivityFX)))))))

(event Custom|ReactivityFX
  (Rendering|Material|SetScalarParameterValueonMaterials (CustomCharacter|GetSkeletalMeshForPartOnPlayerPawn) "Timestamp" (Utilities|Time|GetGameTimeinSeconds))
  (bind _result (Math|Integer|IncrementInt (Variables|Default|GetEliminations))
    (:%20%20
      (Effects|Components|ParticleSystem|SetIntParameter (CallFunction|GetIdleVFXComponent) "Eliminations" _result))))

(fn ConstructionScript
  (|CallParentFunction))
