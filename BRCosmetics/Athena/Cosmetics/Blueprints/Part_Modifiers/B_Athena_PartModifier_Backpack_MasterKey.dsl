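; Part modifier for the Master Key back bling, a backpack cosmetic whose look changes with staged cosmetic state.
; Post-init attaches it and its idle VFX to backpack_root; day phase drives the VFX LightSpawnRate.
; Each stage activation spawns sparks and dispatches to Stage 0-3: 0 resets material and stops VFX, 1-3 toggle emitters if cosmetic events can execute.
; Stage 2 also applies a time-stamped, hue-shifted dissolve material and spawns the key animation.
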(event EventOnPostResIn
  (bind _returnvalue (CustomCharacter|GetSkeletalMeshForPartOnPlayerPawn))
  (bind _returnvalue_2 (CallFunction|GetIdleVFXComponent))
  (|CallParentFunction)
  (Transformation|AttachActorToComponent _returnvalue "backpack_root" "SnapToTarget" "SnapToTarget" "SnapToTarget")
  (Rendering|Material|SetMaterial _returnvalue 0 "/BRCosmetics/Accessories/FORT_Backpacks/Backpack_Master_Key/Materials/BackPack_Master_Key.BackPack_Master_Key")
  (Transformation|AttachComponentToComponent _returnvalue_2 _returnvalue "backpack_root" "SnapToTarget" "SnapToTarget" "SnapToTarget" true)
  (Effects|Components|ParticleSystem|SetEmitterEnable _returnvalue_2 "S1_Ashes" false)
  (Effects|Components|ParticleSystem|SetEmitterEnable _returnvalue_2 "S2_LightAndMist" false)
  (Effects|Components|ParticleSystem|SetEmitterEnable _returnvalue_2 "S3_SpiralAndAshes" false)
  (Effects|Components|ParticleSystem|SetEmitterEnable _returnvalue_2 "S1_KeyHole" false))

(event EventOnDayPhaseChanged
  (params (CurrentDayPhase :type EFortDayPhase%20Enum :default "Morning") (PreviousDayPhase :type EFortDayPhase%20Enum :default "Morning") (bAtCreation :type Boolean))
  (bind _returnvalue (CallFunction|GetIdleVFXComponent))
  (|CallParentFunction CurrentDayPhase PreviousDayPhase bAtCreation)
  (switch Utilities|FlowControl|Switch|SwitchonEFortDayPhase CurrentDayPhase
    (:Morning
      _)
    (:Day
      _)
    (:Evening
      (Effects|Components|ParticleSystem|SetFloatParameter _returnvalue "LightSpawnRate" 1.0))
    (:Night
      _))
  (Effects|Components|ParticleSystem|SetFloatParameter _returnvalue "LightSpawnRate" 0.0))

(event Custom|Stage%200
  (bind _returnvalue (CustomCharacter|GetAssociatedPlayerPawn))
  (Utilities|IsValid _returnvalue
    (:Is%20Valid
      (Rendering|Material|SetMaterial (Customization|GetSkeletalMeshforPartType _returnvalue (CustomCharacter|GetCharacterPartType) :!class /Script/FortniteGame.FortPlayerPawn) 1 "/BRCosmetics/Accessories/FORT_Backpacks/Backpack_Master_Key/Materials/BackPack_Master_Key.BackPack_Master_Key")
      (Components|Activation|Deactivate (CallFunction|GetIdleVFXComponent)))))

(event EventOnCosmeticStateActivated
  (params (Stat%20Name :type Name :default "None") (Stat%20Value :type Integer) (STAGE%20Num%20Being%20Activated :type Integer))
  (bind _stagenum (Variables|Default|GetStageNum))
  (|CallParentFunction Stat%20Name Stat%20Value STAGE%20Num%20Being%20Activated)
  (Variables|Default|SetStageNum STAGE%20Num%20Being%20Activated)
  (Utilities|FlowControl|Sequence
    (:then_0
      (switch int _stagenum
        (:Default
          (Development|PrintString (Utilities|String|Append "Unhandled Cosmetic Stage " (Utilities|String|ToString(Integer) _stagenum)) true true "(R=1.000000,G=0.000000,B=0.006820,A=1.000000)" 5.0))
        (:0
          (CallFunction|Stage0))
        (:1
          (CallFunction|Stage1))
        (:2
          (CallFunction|Stage2))
        (:3
          (CallFunction|Stage3))))
    (:then_1
      (Effects|SpawnFXSystemAttached "/BRCosmetics/Accessories/FORT_Backpacks/Backpack_Master_Key/FX/NS_Backpack_MasterKey_Sparks.NS_Backpack_MasterKey_Sparks" (CallFunction|GetIdleVFXComponent)))))

(event Custom|Stage%201
  (bind _returnvalue (CallFunction|GetIdleVFXComponent))
  (bind _stagenum (Variables|Default|GetStageNum))
  (bind _returnvalue_2 (Utilities|String|ToString(Integer) _stagenum))
  (Networking|CanExecuteCosmeticEvents
    (:True
      (Development|PrintString (Utilities|String|Append _returnvalue_2 "Activate Success!") true true "(R=1.000000,G=0.768187,B=0.000000,A=1.000000)" 4.0)
      ;[0.6416 1 0.6747] Force Activate it to fix the issue of FORT- 150000
      (Components|Activation|Activate _returnvalue true)
      (Effects|Components|ParticleSystem|SetEmitterEnable _returnvalue "S1_Ashes" true)
      (Effects|Components|ParticleSystem|SetEmitterEnable _returnvalue "S1_KeyHole" true)
      (Effects|Components|ParticleSystem|SetEmitterEnable _returnvalue "S2_LightAndMist" false)
      (Effects|Components|ParticleSystem|SetEmitterEnable _returnvalue "S3_SpiralAndAshes" false))
    (:False
      (Development|PrintString (Utilities|String|Append _returnvalue_2 "Activate Failed!") true true "(R=1.000000,G=0.052632,B=0.000000,A=1.000000)" 4.0))))

(event Custom|Stage%202
  (bind _returnvalue (CallFunction|GetIdleVFXComponent))
  (bind _stagenum (Variables|Default|GetStageNum))
  (bind _returnvalue_2 (Utilities|String|ToString(Integer) _stagenum))
  (bind _returnvalue_3 (CustomCharacter|GetAssociatedPlayerPawn))
  (Networking|CanExecuteCosmeticEvents
    (:True
      (Development|PrintString (Utilities|String|Append _returnvalue_2 "Activate Success!") true true "(R=1.000000,G=0.768187,B=0.000000,A=1.000000)" 4.0)
      (Utilities|IsValid _returnvalue_3
        (:Is%20Valid
          (bind _returnvalue_4 (Rendering|Material|CreateDynamicMaterialInstance (Customization|GetSkeletalMeshforPartType _returnvalue_3 (CustomCharacter|GetCharacterPartType) :!class /Script/FortniteGame.FortPlayerPawn) 1 "/BRCosmetics/Accessories/FORT_Backpacks/Backpack_Master_Key/FX/M_Dissolve.M_Dissolve" "None" :!class /Script/Engine.PrimitiveComponent))
          (Rendering|Material|SetScalarParameterValue _returnvalue_4 "BP_DissolveTime" (Utilities|Time|GetGameTimeinSeconds) :!class /Script/Engine.MaterialInstanceDynamic)
          (Rendering|Material|SetScalarParameterValue _returnvalue_4 "BP_HueShift" (Variables|Default|GetHueShift) :!class /Script/Engine.MaterialInstanceDynamic)
          ;[0.6416 1 0.6747] Force Activate it to fix the issue of FORT- 150000
          (Components|Activation|Activate _returnvalue true)
          (Effects|Components|ParticleSystem|SetEmitterEnable _returnvalue "S1_Ashes" true)
          (Effects|Components|ParticleSystem|SetEmitterEnable _returnvalue "S2_LightAndMist" true)
          (Effects|Components|ParticleSystem|SetEmitterEnable _returnvalue "S3_SpiralAndAshes" false)
          (Effects|Components|ParticleSystem|SetEmitterEnable _returnvalue "S1_KeyHole" false)
          (PartModifier|SpawnSystem|SpawnEmitterAttachedPartModifier(Legacy) :EmitterTemplate "/BRCosmetics/Accessories/FORT_Backpacks/Backpack_Master_Key/FX/P_KeyAnimation.P_KeyAnimation" :AttachToComponent (CustomCharacter|GetSkeletalMeshForPartOnPlayerPawn) :AttachPointName "backpack_root" :Scale "1.000000,1.000000,1.000000" :LocationType "SnapToTargetIncludingScale" :bAutoDestroy true :PoolingMethod "None" :bAutoActivate true :bMatchPawnCustomDepth true :bDebounce true :DebounceTime -1.0))))
    (:False
      (Development|PrintString (Utilities|String|Append _returnvalue_2 "Activate Failed!") true true "(R=1.000000,G=0.052632,B=0.000000,A=1.000000)" 4.0))))

(event Custom|Stage%203
  (bind _returnvalue (CallFunction|GetIdleVFXComponent))
  (bind _stagenum (Variables|Default|GetStageNum))
  (bind _returnvalue_2 (Utilities|String|ToString(Integer) _stagenum))
  (Networking|CanExecuteCosmeticEvents
    (:True
      (Development|PrintString (Utilities|String|Append _returnvalue_2 "Activate Success!") true true "(R=1.000000,G=0.768187,B=0.000000,A=1.000000)" 4.0)
      (Components|Activation|Activate _returnvalue true)
      (Effects|Components|ParticleSystem|SetEmitterEnable _returnvalue "S1_Ashes" true)
      (Effects|Components|ParticleSystem|SetEmitterEnable _returnvalue "S2_LightAndMist" true)
      (Effects|Components|ParticleSystem|SetEmitterEnable _returnvalue "S3_SpiralAndAshes" true)
      (Effects|Components|ParticleSystem|SetEmitterEnable _returnvalue "S1_KeyHole" false))
    (:False
      (Development|PrintString (Utilities|String|Append _returnvalue_2 "Activate Failed!") true true "(R=1.000000,G=0.052632,B=0.000000,A=1.000000)" 4.0))))

(fn ConstructionScript
  (|CallParentFunction))
