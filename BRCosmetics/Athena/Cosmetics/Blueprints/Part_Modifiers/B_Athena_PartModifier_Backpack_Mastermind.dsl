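; Cosmetic part modifier Blueprint for the Mastermind backpack that adjusts its idle particle effect as the pawn moves between grounded and airborne states.
; Jumping and beginning a skydive set the idle VFX component's Spawn parameter to zero and clear its Velocity vector.
; Landing and ending a skydive restore Spawn to one and apply a small negative X/Y Velocity.
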(event EventOnPawnJumped
  (bind _returnvalue (CallFunction|GetIdleVFXComponent))
  (Effects|Components|ParticleSystem|SetVectorParameter _returnvalue "Velocity" "0, 0, 0")
  (Effects|Components|ParticleSystem|SetFloatParameter _returnvalue "Spawn" 0.0))

(event EventOnBeginSkydiving
  (bind _returnvalue (CallFunction|GetIdleVFXComponent))
  (Effects|Components|ParticleSystem|SetVectorParameter _returnvalue "Velocity" "0, 0, 0")
  (Effects|Components|ParticleSystem|SetFloatParameter _returnvalue "Spawn" 0.0))

(event EventOnPawnLanded
  (bind _returnvalue (CallFunction|GetIdleVFXComponent))
  (Effects|Components|ParticleSystem|SetVectorParameter _returnvalue "Velocity" "-0.200000,-0.200000,0.000000")
  (Effects|Components|ParticleSystem|SetFloatParameter _returnvalue "Spawn" 1.0))

(event EventOnEndSkydiving
  (bind _returnvalue (CallFunction|GetIdleVFXComponent))
  (Effects|Components|ParticleSystem|SetVectorParameter _returnvalue "Velocity" "-0.200000,-0.200000,0.000000")
  (Effects|Components|ParticleSystem|SetFloatParameter _returnvalue "Spawn" 1.0))

(fn ConstructionScript)
