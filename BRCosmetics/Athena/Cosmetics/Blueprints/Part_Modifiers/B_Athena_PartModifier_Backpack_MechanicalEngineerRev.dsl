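; Backpack part modifier for the MechanicalEngineerRev cosmetic that toggles the backpack's idle VFX as the player changes weapons.
; On weapon change, IsMechEngineerPickaxe searches the weapon's skeletal mesh name for the matching axe and, despite its name, returns true when absent.
; HideBackpackFX activates the effect on true and deactivates it on false, so the axe or an invalid weapon or mesh turns it off.
; A debug PrintString remains in the check.
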(event EventOnWeaponChanged
  (params (NewWeapon :type Fort%20Weapon%20Object%20Reference) (OldWeapon :type Fort%20Weapon%20Object%20Reference))
  (bind _is_weapon (CallFunction|IsMechEngineerPickaxe NewWeapon))
  (CallFunction|HideBackpackFX _is_weapon))

(event EventOnDayPhaseChanged
  (params (CurrentDayPhase :type EFortDayPhase%20Enum :default "Morning") (PreviousDayPhase :type EFortDayPhase%20Enum :default "Morning") (bAtCreation :type Boolean)))

(event EventOnNewPlayerWeaponType
  (params (NewType :type EFortItemType%20Enum :default "WorldItem") (OldType :type EFortItemType%20Enum :default "WorldItem"))
  (|CallParentFunction NewType OldType) ; This node is disabled and will not be called.\nDrag off pins to build functionality.
)

(event EventOnWeaponFired)

(event EventOnNewWeapons
  (params (WeaponsAsItems :type Array%20of%20Fort%20Item%20Object%20References)))

(event EventOnCurieStateAttached
  (params (Tag :type Gameplay%20Tag%20Structure%20%28by%20ref%29)))

(event EventOnCurieStateDetached
  (params (Tag :type Gameplay%20Tag%20Structure%20%28by%20ref%29)))

(event EventOnBeginSkydiving
  (|CallParentFunction) ; This node is disabled and will not be called.\nDrag off pins to build functionality.
)

(event EventOnEndSkydiving
  (|CallParentFunction) ; This node is disabled and will not be called.\nDrag off pins to build functionality.
)

(event EventOnEnterVehicle
  (params (VehicleEntered :type Fort%20Vehicle%20Interface%20Interface%20%28by%20ref%29))
  (|CallParentFunction VehicleEntered) ; This node is disabled and will not be called.\nDrag off pins to build functionality.
)

(event EventOnExitVehicle
  (|CallParentFunction) ; This node is disabled and will not be called.\nDrag off pins to build functionality.
)

(event EventOnTargetingOrFirstPersonCameraChanged)

(event EventOnHitPawn
  (params (SelfActor :type Actor%20Object%20Reference) (OtherActor :type Actor%20Object%20Reference) (NormalImpulse :type Vector :default "0, 0, 0") (Hit :type Hit%20Result%20Structure%20%28by%20ref%29)))

(fn HideBackpackFX
  (params (Visibility :type Boolean))
  (bind _returnvalue (CallFunction|GetIdleVFXComponent))
  (if Visibility
    (Components|Activation|Activate _returnvalue false)
    (else
      (Components|Activation|Deactivate _returnvalue))))

(fn IsMechEngineerPickaxe
  (params (Weapon :type Fort%20Weapon%20Object%20Reference))
  (bind _returnvalue (Visual|GetWeaponMesh Weapon))
  (bind _returnvalue_2 (Components|SkeletalMesh|GetSkeletalMeshAsset _returnvalue))
  (bind _returnvalue_3 (Utilities|GetObjectName _returnvalue_2))
  (bind _returnvalue_4 (Utilities|String|Contains _returnvalue_3 "MechanicalEngineerRev_Axe"))
  (Utilities|IsValid Weapon
    (:Is%20Valid
      (Utilities|IsValid _returnvalue
        (:Is%20Valid
          (Utilities|IsValid _returnvalue_2
            (:Is%20Valid
              (Development|PrintString (Utilities|String|ToString(Boolean) _returnvalue_4) true true "(R=1.000000,G=0.063568,B=0.000000,A=1.000000)" 25.0)
              (return (not _returnvalue_4)))
            (:Is%20Not%20Valid
              _)))
        (:Is%20Not%20Valid
          _)))
    (:Is%20Not%20Valid
      _))
  (return false))

(fn ConstructionScript
  (|CallParentFunction))
