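; Backpack part modifier for the Meteorite cosmetic, handling its debris component and idle particle effect.
; After res-in, it attaches the meteor debris component to the player pawn's part skeletal mesh at FX_Socket, snapping to the socket.
; Skydiving events toggle the idle VFX SpawnPerUnitRate parameter, zeroing it when skydiving begins and restoring it to one when skydiving ends.
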(event EventOnPostResIn
  (|CallParentFunction)
  (Transformation|AttachComponentToComponent (Variables|Default|Getsm_meteor_Backpack_Debris) (CustomCharacter|GetSkeletalMeshForPartOnPlayerPawn) "FX_Socket" "SnapToTarget" "SnapToTarget" "SnapToTarget" true))

(event EventOnBeginSkydiving
  (|CallParentFunction)
  (Effects|Components|ParticleSystem|SetFloatParameter (CallFunction|GetIdleVFXComponent) "SpawnPerUnitRate" 0.0))

(event EventOnEndSkydiving
  (|CallParentFunction)
  (Effects|Components|ParticleSystem|SetFloatParameter (CallFunction|GetIdleVFXComponent) "SpawnPerUnitRate" 1.0))

(fn ConstructionScript)
