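; Backpack part modifier Blueprint for the MilitaryFashion Brigade cosmetic in Athena.
; When the pawn lands, it writes the current game time to the LandGameTime float parameter on the idle VFX component, giving that particle effect a timestamp of the landing.
; The post-res-in event and construction script only call their parent implementations.
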(event EventOnPostResIn
  (|CallParentFunction))

(event EventOnPawnLanded
  (|CallParentFunction)
  (Effects|Components|ParticleSystem|SetFloatParameter (CallFunction|GetIdleVFXComponent) "LandGameTime" (Utilities|Time|GetGameTimeinSeconds)))

(fn ConstructionScript
  (|CallParentFunction))
