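; Part modifier Blueprint for the Moss backpack cosmetic that makes its idle VFX react when the pawn lands.
; On the pawn-landed event, it calls the parent handler, then sets a "Landing" bool parameter on the idle VFX particle component to true.
; After a short delay it resets the parameter to false, producing a brief pulse.
; The construction script only calls its parent.
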(event EventOnPawnLanded
  (bind _returnvalue (CallFunction|GetIdleVFXComponent))
  (|CallParentFunction)
  (Effects|Components|ParticleSystem|SetBoolParameter _returnvalue "Landing" true)
  (Utilities|FlowControl|Delay 0.1333)
  (Effects|Components|ParticleSystem|SetBoolParameter _returnvalue "Landing" false))

(fn ConstructionScript
  (|CallParentFunction))
