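; Part modifier for the Ms Whip backpack cosmetic that swaps its particle effects as the player pawn jumps and lands.
; On jump it spawns a dribble emitter at the backpack mesh's FX_Base socket and deactivates the idle VFX component.
; On landing it spawns a burst emitter at the same socket and reactivates the idle VFX.
; Each successfully spawned emitter receives the pawn's backpack-slot cosmetic variants.
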(event EventOnPawnJumped
  (bind _returnvalue (PartModifier|SpawnSystem|SpawnEmitterAttachedPartModifier(Legacy) "/BRCosmetics/Accessories/FORT_Backpacks/Backpack_F_MED_Ms_Whip/Effects/P_Backpack_MsWhip_Dribble.P_Backpack_MsWhip_Dribble" (CustomCharacter|GetSkeletalMeshForPartOnPlayerPawn) "FX_Base")
    (:Success
      (Class|FortCosmeticFunctionLibrary|ApplyVariantstoComponentbySlotTag (CustomCharacter|GetAssociatedPlayerPawn) "(TagName=\"LoadoutSlot.Backpack\")" _returnvalue
        (:True
          _)
        (:False
          _)))
    (:Failure
      _))
  (Components|Activation|Deactivate (CallFunction|GetIdleVFXComponent)))

(event EventOnPawnLanded
  (bind _returnvalue (PartModifier|SpawnSystem|SpawnEmitterAttachedPartModifier(Legacy) "/BRCosmetics/Accessories/FORT_Backpacks/Backpack_F_MED_Ms_Whip/Effects/P_Backpack_MsWhip_Burst.P_Backpack_MsWhip_Burst" (CustomCharacter|GetSkeletalMeshForPartOnPlayerPawn) "FX_Base")
    (:Success
      (Class|FortCosmeticFunctionLibrary|ApplyVariantstoComponentbySlotTag (CustomCharacter|GetAssociatedPlayerPawn) "(TagName=\"LoadoutSlot.Backpack\")" _returnvalue
        (:True
          _)
        (:False
          _)))
    (:Failure
      _))
  (Components|Activation|Activate (CallFunction|GetIdleVFXComponent) false))

(fn ConstructionScript)
