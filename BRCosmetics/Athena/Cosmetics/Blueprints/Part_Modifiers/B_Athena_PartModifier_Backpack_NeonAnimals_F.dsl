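; Backpack part modifier for the Neon Animals cosmetic that gives swap feedback when the player's weapon type changes.
; Melee and ranged weapons set ColorLerp to opposite ends.
; A particle burst attaches to the part's skeletal mesh root, tinted by blending pink to blue with that value through a vector parameter.
; ColorLerp also selects which of two swap sounds plays at the part mesh's location.
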(event EventActuallyChangedWeaponType
  (params (New%20Type :type EFortItemType%20Enum%20%28by%20ref%29 :default "WorldItem"))
  (bind _colorlerp (Variables|Default|GetColorLerp))
  (switch Utilities|FlowControl|Switch|SwitchonEFortItemType New%20Type
    (:WeaponMelee
      (Variables|Default|SetColorLerp 1.0))
    (:WeaponRanged
      (Variables|Default|SetColorLerp 0.0))
    (:WeaponHarvest
      _))
  (bind _returnvalue (PartModifier|SpawnSystem|SpawnEmitterAttachedPartModifier(Legacy) "/BRCosmetics/Effects/Fort_Effects/Effects/Characters/Athena_Parts/Neon/P_NeonAnimals_SwapPack.P_NeonAnimals_SwapPack" (CustomCharacter|GetSkeletalMeshForPartOnPlayerPawn) "root")
    (:Success
      (Effects|Components|ParticleSystem|SetVectorParameter _returnvalue "InstanceColorScaleOverLife" (Math|Conversions|ToVector(LinearColor) (Math|Color|Lerp(LinearColor) "(R=1.500000,G=0.244932,B=0.639988,A=1.000000)" "(R=0.000000,G=0.171109,B=1.000000,A=1.000000)" _colorlerp))))
    (:Failure
      _))
  (Audio|PlaySoundatLocation (select (== _colorlerp 0.0) "/BRCosmetics/Sounds/Backblings/SkinTransition/Quest_NeonAnimal_Swap_Weapon_Cue.Quest_NeonAnimal_Swap_Weapon_Cue" "/BRCosmetics/Sounds/Backblings/SkinTransition/Quest_NeonAnimal_Swap_Cue.Quest_NeonAnimal_Swap_Cue") (Transformation|GetWorldLocation (Customization|GetSkeletalMeshforPartType (CustomCharacter|GetAssociatedPlayerPawn) (CustomCharacter|GetCharacterPartType) :!class /Script/FortniteGame.FortPlayerPawn))))

(fn ConstructionScript)
