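; Backpack cosmetic part modifier that shows the player's Athena kill count as a two-digit counter on the part's mesh materials.
; On pawn/part change it reads and watches the kill stat, refreshing on each change and after res-in.
; UpdateMaterialParameters splits the count into tens and ones scalars with a single-digit toggle, skipping writes above 99.
; Stat changes also play a timeline driving the FlashControl parameter, reset to zero when finished.
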(event EventOnPawnAndPartChanged
  (bind _returnvalue (CustomCharacter|GetAssociatedPlayerPawn))
  (bind _returnvalue_2 (CustomCharacter|GetSkeletalMeshForPartOnPlayerPawn))
  (|CallParentFunction)
  (bind _returnvalue_3 (Class|FortCosmeticFunctionLibrary|GetCosmeticStatValue _returnvalue "(TagName=\"StatManager.AthenaKills\")"))
  ;[1 1 1] Simple, 2 digit kill counter designed for material use
  (Variables|Default|SetKillCount (Math|Conversions|ToFloat(Integer) _returnvalue_3))
  (bind (_statname _statvalue) (Cosmetics|WatchCosmeticStat _returnvalue "(TagName=\"StatManager.AthenaKills\")" 0 false)
    (:then
      (CallFunction|UpdateMaterialParameters))
    (:OnStatChanged
      (Variables|Default|SetKillCount (Math|Conversions|ToFloat(Integer) _statvalue))
      (CallFunction|UpdateMaterialParameters)
      (bind (_direction _lightconrol) (Timeline|UpdateNumber PlayFromStart)
        (:Update
          (Rendering|Material|SetScalarParameterValueonMaterials _returnvalue_2 "FlashControl" _lightconrol))
        (:Finished
          (Rendering|Material|SetScalarParameterValueonMaterials _returnvalue_2 "FlashControl" 0.0))))))

(event EventOnPostResIn
  (|CallParentFunction)
  (CallFunction|UpdateMaterialParameters))

(fn UpdateMaterialParameters
  (bind _returnvalue (CustomCharacter|GetSkeletalMeshForPartOnPlayerPawn))
  (bind _killcount (Variables|Default|GetKillCount))
  (bind _returnvalue_2 (/ _killcount 1e+01))
  (bind _returnvalue_3 (Math|Float|Fraction _returnvalue_2))
  (bind _returnvalue_4 (* _returnvalue_3 1e+01))
  (bind _killcount_2 (Variables|Default|GetKillCount))
  (if (> _killcount_2 0.0)
    (if (> _killcount_2 9.0)
      (if (not (> _killcount_2 99.0))
        (Rendering|Material|SetScalarParameterValueonMaterials _returnvalue "UseSingleDigit" 0.0)
        (Rendering|Material|SetScalarParameterValueonMaterials _returnvalue "Kills-Tens" (Math|Conversions|ToFloat(Integer) (Math|Float|Floor _returnvalue_2)))
        (Rendering|Material|SetScalarParameterValueonMaterials _returnvalue "Kills-Ones" _returnvalue_4))
      (else
        ;[1 1 1] Use single digit
        (Rendering|Material|SetScalarParameterValueonMaterials _returnvalue "UseSingleDigit" 1.0)
        (Rendering|Material|SetScalarParameterValueonMaterials _returnvalue "Kills-Tens" 0.0)
        (Rendering|Material|SetScalarParameterValueonMaterials _returnvalue "Kills-Ones" _returnvalue_4)))
    (else
      (Rendering|Material|SetScalarParameterValueonMaterials _returnvalue "UseSingleDigit" 1.0)
      (Rendering|Material|SetScalarParameterValueonMaterials _returnvalue "Kills-Tens" 0.0)
      (Rendering|Material|SetScalarParameterValueonMaterials _returnvalue "Kills-Ones" 0.0))))

(fn ConstructionScript)
