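; Part modifier for the NickelScan back bling that drives reactive material stages from the player's elimination count.
; On res-in, a frontend timeline previews three stages, while in game a kill-stat watcher writes eliminations directly to CurrentStage.
; OnActualStatChange stamps game time into the material to time the animation.
; The threshold-based stage path through UpdateMaterialParameters ends in empty StageOne/Two/Three events, so it changes nothing visible.
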(event EventOnCosmeticStateActivated
  (params (Stat%20Name :type Name :default "None") (Stat%20Value :type Integer) (STAGE%20Num%20Being%20Activated :type Integer))
  (|CallParentFunction Stat%20Name Stat%20Value STAGE%20Num%20Being%20Activated)
  (Variables|Default|SetCurrentStage STAGE%20Num%20Being%20Activated)
  (CallFunction|UpdateMaterialParameters))

(event EventPreDestruction
  (|CallParentFunction)
  (Utilities|Time|ClearandInvalidateTimerbyHandle (Variables|Default|GetRefreshTimeHandler)))

(event EventOnPawnAndPartChanged
  (|CallParentFunction)
  (if (not (Reactive|Frontend|ShouldDisplayFrontendReactivity 0 "Backpack"))
    (bind _returnvalue (Class|FortCosmeticFunctionLibrary|GetCosmeticStatValue (CustomCharacter|GetAssociatedPlayerPawn) "(TagName=\"StatManager.AthenaKills\")"))
    (bind _stage (CallFunction|StatValuetoStage _returnvalue))
    (Variables|Default|SetCurrentStage _stage)
    (CallFunction|UpdateMaterialParameters)))

(event EventOnPostResIn
  (bind _backpackskelmesh (Variables|Default|GetBackpackSkelMesh))
  (|CallParentFunction)
  (Variables|Default|SetBackpackSkelMesh (CustomCharacter|GetSkeletalMeshForPartOnPlayerPawn))
  (if (Reactive|Frontend|ShouldDisplayFrontendReactivity 0 "Backpack")
    ; In FE mode: Its gonna be 3 stage represents Normal > Emissive light up > FX up
    (Timeline|Timeline_0 PlayFromStart
      (:Stage%201
        (Rendering|Material|SetScalarParameterValueonMaterials _backpackskelmesh "Animation_Duration" 2.0)
        (Rendering|Material|SetScalarParameterValueonMaterials _backpackskelmesh "CurrentStage" 1.0))
      (:Stage%202
        (Rendering|Material|SetScalarParameterValueonMaterials _backpackskelmesh "Animation_Duration" 2.0)
        (Rendering|Material|SetScalarParameterValueonMaterials _backpackskelmesh "CurrentStage" 2.0))
      (:Stage%203
        (Rendering|Material|SetScalarParameterValueonMaterials _backpackskelmesh "Animation_Duration" 2.0)
        (Rendering|Material|SetScalarParameterValueonMaterials _backpackskelmesh "CurrentStage" 3.0)))
    (CallFunction|OnActualStatChange)
    (else
      ; In non FE mode: Its gonna be 5 stages represents 5 eliminates
      (bind (_statname _statvalue) (Cosmetics|WatchCosmeticStat (CustomCharacter|GetAssociatedPlayerPawn) "(TagName=\"StatManager.AthenaKills\")")
        (:OnStatChanged
          (Variables|Default|SetEliminations _statvalue)
          (Utilities|FlowControl|Sequence
            (:then_0
              (Rendering|Material|SetScalarParameterValueonMaterials (Variables|Default|GetBackpackSkelMesh) "Animation_Duration" (Variables|Default|GetAnimation_Duration))
              (Rendering|Material|SetScalarParameterValueonMaterials (Variables|Default|GetBackpackSkelMesh) "CurrentStage" (Math|Conversions|ToFloat(Integer) (Variables|Default|GetEliminations))))
            (:then_1
              (CallFunction|OnActualStatChange))))))))

(event Custom|OnActualStatChange
  ; In non FE mode: Its gonna be 5 stages represents 5 eliminates
  ;;[1 1 1] This should only fire off when the stat changes - use this event!
  (Rendering|Material|SetScalarParameterValueonMaterials (Variables|Default|GetBackpackSkelMesh) "GameTimeFromBP" (Utilities|Time|GetGameTimeinSeconds)))

(event Custom|StageOne)

(event Custom|PlayReactiveAudio
  (params (Event%20Name :type Gameplay%20Tag%20Structure)))

(event Custom|StageTwo)

(event Custom|StageThree)

(fn StatValueToStage
  (params (StatValue :type Integer))
  (bind (_arrayindex _arrayelement) (Utilities|Array|ReverseforEachLoop (Variables|PartModifierControls|WatchCosmeticStat|GetStat%20Thresholds))
    (:LoopBody
      (if (>= (Variables|GetStatValue) _arrayelement)
        (return _arrayindex)))
    (:Completed
      (return 0))))

(fn UpdateMaterialParameters
  (switch int (Variables|Default|GetCurrentStage)
    (:0
      (CallFunction|StageOne))
    (:1
      (CallFunction|StageTwo))
    (:2
      (CallFunction|StageThree))))

(fn ConstructionScript
  (bind _returnvalue (CustomCharacter|GetSkeletalMeshForPartOnPlayerPawn))
  (|CallParentFunction)
  (Rendering|Material|SetScalarParameterValueonMaterials _returnvalue "GameTimeFromBP" -99999.0)
  (Rendering|Material|SetScalarParameterValueonMaterials _returnvalue "StartAtStage" (Math|Conversions|ToFloat(Integer) 0))
  (Rendering|Material|SetScalarParameterValueonMaterials _returnvalue "StopAtStage" (Math|Conversions|ToFloat(Integer) 0)))
