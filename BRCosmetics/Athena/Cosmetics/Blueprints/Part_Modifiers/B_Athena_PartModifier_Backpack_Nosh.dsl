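; Backpack part modifier for the Nosh cosmetic, managing mouth particle effects on the player pawn.
; After res-in, it snaps the mouth effect component to the backpack mesh's FX_Mouth socket, and hides it while skydiving.
; On jump, it activates the idle VFX component and spawns a debounced, auto-destroying jump emitter at that socket.
; Each event calls the parent, whose skydiving FX toggle is noted as enabled through class defaults.
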(event EventOnBeginSkydiving
  ;[1 1 1] Turn off/on fx when skydiving/stopped skydiving, done on the parent, enabled in class defaults.
  (|CallParentFunction)
  (Development|SetHiddeninGame (Variables|Default|GetP_Nosh_Mouth_Backpack) true false))

(event EventOnEndSkydiving
  ;[1 1 1] Turn off/on fx when skydiving/stopped skydiving, done on the parent, enabled in class defaults.
  (|CallParentFunction)
  (Development|SetHiddeninGame (Variables|Default|GetP_Nosh_Mouth_Backpack) false false))

(event EventOnPawnJumped
  (|CallParentFunction)
  (Components|Activation|Activate (CallFunction|GetIdleVFXComponent) true)
  (PartModifier|SpawnSystem|SpawnEmitterAttachedPartModifier(Legacy) :EmitterTemplate "/BRCosmetics/Accessories/FORT_Backpacks/Backpack_M_MED_Nosh/FX/P_Nosh_Mouth_Backpack_Jumping.P_Nosh_Mouth_Backpack_Jumping" :AttachToComponent (CustomCharacter|GetSkeletalMeshForPartOnPlayerPawn) :AttachPointName "FX_Mouth" :Scale "1.000000,1.000000,1.000000" :LocationType "SnapToTargetIncludingScale" :bAutoDestroy true :PoolingMethod "None" :bAutoActivate true :bMatchPawnCustomDepth true :bDebounce true :DebounceTime -1.0))

(event EventOnPostResIn
  (|CallParentFunction)
  (Transformation|AttachComponentToComponent (Variables|Default|GetP_Nosh_Mouth_Backpack) (CustomCharacter|GetSkeletalMeshForPartOnPlayerPawn) "FX_Mouth" "SnapToTarget" "SnapToTarget" "SnapToTarget" true))

(fn ConstructionScript)
