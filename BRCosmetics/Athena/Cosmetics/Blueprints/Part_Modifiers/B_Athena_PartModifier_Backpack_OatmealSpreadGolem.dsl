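; Part modifier Blueprint for the OatmealSpreadGolem reactive back bling, which responds to the player's elimination count.
; In game, a kill stat watcher drives a clamped "Score" material parameter, time-stamped for shader animation, plus a Niagara "KillCount".
; In-range stat changes also trigger Sound Library audio that cycles through tags, with fade-out and retrigger gating.
; In the frontend, a looping timer previews score stages with random kill counts instead.
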(event EventOnCosmeticStateActivated
  (params (Stat%20Name :type Name :default "None") (Stat%20Value :type Integer) (STAGE%20Num%20Being%20Activated :type Integer))
  (|CallParentFunction Stat%20Name Stat%20Value STAGE%20Num%20Being%20Activated)
  (Variables|Default|SetCurrentScore STAGE%20Num%20Being%20Activated))

(event EventPreDestruction
  (|CallParentFunction)
  (Utilities|Time|ClearandInvalidateTimerbyHandle (Variables|Default|GetFE_Loop_TimerHandle)))

(event EventOnPawnAndPartChanged
  (|CallParentFunction)
  (if (not (Reactive|Frontend|ShouldDisplayFrontendReactivity 0 "Backpack"))
    (bind _returnvalue (Class|FortCosmeticFunctionLibrary|GetCosmeticStatValue (CustomCharacter|GetAssociatedPlayerPawn) "(TagName=\"StatManager.AthenaKills\")"))
    (bind _stage (CallFunction|StatValuetoStage _returnvalue))
    (Variables|Default|SetCurrentScore _stage)))

(event Custom|CustomEvent
  (bind _currentscore (Variables|Default|GetCurrentScore))
  ;[0.000129 0.2726 1 0.5] In FE mode: Score Loop
  (Utilities|FlowControl|Sequence
    (:then_0
      (if (< (Variables|Default|GetCurrentScore) (+ (Variables|Default|GetEndScore) 1))
        (CallFunction|FrontEndScoreTransition (Math|Conversions|ToFloat(Integer) (Variables|Default|GetCurrentScore)))
        (else
          _))
      (Math|Integer|IncrementInt _currentscore
        (:%20%20
          (if (> _currentscore (+ (Variables|Default|GetEndScore) 1))
            (Utilities|Time|ClearandInvalidateTimerbyHandle (Variables|Default|GetFE_Loop_TimerHandle))
            (Variables|Default|SetCurrentScore 1)
            (CallFunction|FrontEndScoreTransition 0.0)
            (bind _returnvalue (Utilities|Time|SetTimerbyEvent (delegate Custom|CustomEvent) 4.0 true))
            (Variables|Default|SetFE_Loop_TimerHandle _returnvalue)))))
    (:then_1
      (bind _asniagara_particle_system_component (Utilities|Casting|CastToNiagaraComponent (CallFunction|GetIdleVFXComponent)))
      (Niagara|SetNiagaraVariable(Int32) _asniagara_particle_system_component "KillCount" (Math|Random|RandomInteger 99999)))))

(event Custom|FrontEndScoreTransition
  (params (Score :type Float%20%28single-precision%29))
  (bind _returnvalue (CustomCharacter|GetSkeletalMeshForPartOnPlayerPawn))
  ;[0 0.3267 1 0.5] FrontEnd Score Transition
  (Rendering|Material|SetScalarParameterValueonMaterials _returnvalue "GameTimeFromBP" (Utilities|Time|GetGameTimeinSeconds))
  (Rendering|Material|SetScalarParameterValueonMaterials _returnvalue "Animation_Duration" (Variables|Default|GetPerScoreDuration))
  (Utilities|FlowControl|Delay 0.5)
  (Rendering|Material|SetScalarParameterValueonMaterials _returnvalue "Score" Score))

(event EventOnPostResIn
  (bind _returnvalue (CustomCharacter|GetSkeletalMeshForPartOnPlayerPawn))
  (bind _endscore (Variables|Default|GetEndScore))
  (bind _returnvalue_2 (+ _endscore 1))
  (bind _lastscore (Variables|Default|GetLastScore))
  (|CallParentFunction)
  (if (Reactive|Frontend|ShouldDisplayFrontendReactivity 0 "Backpack")
    ;[0.000129 0.2726 1 0.5] In FE mode: Score Loop
    (Variables|Default|SetCurrentScore 1)
    (CallFunction|FrontEndScoreTransition 0.0)
    (bind _returnvalue_3 (Utilities|Time|SetTimerbyEvent (delegate Custom|CustomEvent) 4.0 true))
    (Variables|Default|SetFE_Loop_TimerHandle _returnvalue_3)
    (else
      ; In non FE mode: Its gonna be 5 stages represents 5 eliminates
      (bind (_statname _statvalue) (Cosmetics|WatchCosmeticStat (CustomCharacter|GetAssociatedPlayerPawn) "(TagName=\"StatManager.AthenaKills\")" 0 false)
        (:OnStatChanged
          (Variables|Default|SetLastScore (Math|Integer|SelectInt (Math|Integer|Max(Integer) _statvalue _lastscore) _endscore (< (Math|Integer|Max(Integer) _statvalue _lastscore) _returnvalue_2)))
          (Utilities|FlowControl|Sequence
            (:then_0
              (Rendering|Material|SetScalarParameterValueonMaterials _returnvalue "Animation_Duration" (Variables|Default|GetPerScoreDuration))
              (Utilities|FlowControl|Delay 0.5)
              (Rendering|Material|SetScalarParameterValueonMaterials _returnvalue "Score" (Math|Conversions|ToFloat(Integer) (Variables|Default|GetLastScore))))
            (:then_1
              (if (and (> (Math|Integer|Max(Integer) _statvalue _lastscore) (Variables|Default|GetStartScore)) (< (Math|Integer|Max(Integer) _statvalue _lastscore) _returnvalue_2))
                (CallFunction|OnActualStatChange)))
            (:then_2
              (if (> _statvalue 0)
                (bind _asniagara_particle_system_component (Utilities|Casting|CastToNiagaraComponent (CallFunction|GetIdleVFXComponent)))
                (Niagara|SetNiagaraVariable(Int32) _asniagara_particle_system_component "KillCount" _statvalue)
                (Development|PrintString (Utilities|String|ToString(Integer) _statvalue))))))))))

(event Custom|OnActualStatChange
  ; In non FE mode: Its gonna be 5 stages represents 5 eliminates
  ;;[1 1 1] This should only fire off when the stat changes - use this event!
  (Rendering|Material|SetScalarParameterValueonMaterials (CustomCharacter|GetSkeletalMeshForPartOnPlayerPawn) "GameTimeFromBP" (Utilities|Time|GetGameTimeinSeconds))
  (CallFunction|SetAudioScore))

(event Custom|PlayAudio
  (params (Event%20Name :type Gameplay%20Tag%20Structure))
  ; Reactive Audio
  (bind _returnvalue (SoundLibrary|PlaySoundSimple (WorldSubsystems|GetSoundLibrarySubsystem) self (Utilities|Struct|MakeSimpleSoundSettings Event%20Name true false true true 1.0 1.0 0.0 0 0 "None" (CustomCharacter|GetSkeletalMeshForPartOnPlayerPawn)) :!class /Script/SoundLibrary.SoundLibrarySubsystem))
  (if (Utilities|Array|IsValidIndex (.AudioComponents (Utilities|Struct|BreakSoundLibraryPlayResult _returnvalue)))
    (Variables|Default|SetReactiveAudComp (Utilities|Array|Get(acopy) (.AudioComponents (Utilities|Struct|BreakSoundLibraryPlayResult _returnvalue))))
    (Utilities|Time|SetTimerbyEvent (EventDispatchers|CreateEvent :!event Custom|AllowRetrigger) 0.05)))

(event Custom|AllowRetrigger
  (bind _fade_out_duration (Variables|Default|GetFade%20Out%20Duration))
  (bind _currentaudscore (Variables|Default|GetCurrentAudScore))
  ; Reactive Audio
  (Utilities|FlowControl|Gate true
    (:Exit
      (Audio|Components|Audio|FadeOut (Variables|Default|GetReactiveAudComp) _fade_out_duration 0.0 "Linear" :!class /Script/Engine.AudioComponent)
      (Utilities|FlowControl|Delay _fade_out_duration)
      (CallFunction|PlayAudio (Utilities|Array|Get(acopy) (Variables|Default|GetSoundLibraryTags) (Variables|Default|GetCurrentAudScore)))
      (if (>= _currentaudscore (Variables|Default|GetMaxAudioscore))
        (Variables|Default|SetCurrentAudScore 0)
        (else
          (Math|Integer|IncrementInt _currentaudscore))))))

(event Custom|SetAudioScore
  (bind _fade_out_duration (Variables|Default|GetFade%20Out%20Duration))
  (bind _currentaudscore (Variables|Default|GetCurrentAudScore))
  ; Reactive Audio
  (Networking|CanExecuteCosmeticEvents
    (:True
      (if (not (Utilities|IsValid (Variables|Default|GetReactiveAudComp) :!class /Script/Engine.KismetSystemLibrary))
        (Utilities|FlowControl|Sequence
          (:then_0
            (Utilities|FlowControl|Gate true
              (:Exit
                (Audio|Components|Audio|FadeOut (Variables|Default|GetReactiveAudComp) _fade_out_duration 0.0 "Linear" :!class /Script/Engine.AudioComponent)
                (Utilities|FlowControl|Delay _fade_out_duration))))
          (:then_1
            _))
        (CallFunction|PlayAudio (Utilities|Array|Get(acopy) (Variables|Default|GetSoundLibraryTags) (Variables|Default|GetCurrentAudScore)))
        (if (>= _currentaudscore (Variables|Default|GetMaxAudioscore))
          (Variables|Default|SetCurrentAudScore 0)
          (else
            (Math|Integer|IncrementInt _currentaudscore)))))))

(fn AudioStatWatcher
  (params (StatName :type Name :default "None") (StatValue :type Integer) (StatTag :type Fort%20Stat%20Manager%20Tag%20Structure%20%28by%20ref%29))
  (|CallParentFunction StatName StatValue StatTag)
  (return))

(fn StatValueToStage
  (params (StatValue :type Integer))
  (bind (_arrayindex _arrayelement) (Utilities|Array|ReverseforEachLoop (Variables|PartModifierControls|WatchCosmeticStat|GetStat%20Thresholds))
    (:LoopBody
      (if (>= (Variables|GetStatValue) _arrayelement)
        (return _arrayindex)))
    (:Completed
      (return 0))))

(fn ConstructionScript
  (bind _returnvalue (CustomCharacter|GetSkeletalMeshForPartOnPlayerPawn))
  (|CallParentFunction)
  (Rendering|Material|SetScalarParameterValueonMaterials _returnvalue "GameTimeFromBP" -99999.0)
  (Rendering|Material|SetScalarParameterValueonMaterials _returnvalue "Score" (Math|Conversions|ToFloat(Integer) (Variables|Default|GetStartScore))))
