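; Cosmetic part modifier for the OboeThorn back bling that makes its idle VFX react to player kills and to its matching axe.
; A rise in the watched kill stat stamps game time into the effect and enables reactivity, which a retriggerable delay later switches off.
; In the front end, a looping timer previews that reaction instead.
; On res-in and weapon change, a mesh-name check for the OboeThorn axe sets the effect's weapon-on flag.
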(event Custom|FrontEndReactivity
  ; Front End Loop
  ; FX
  (CallFunction|ReactivityEvent true)
  (Utilities|FlowControl|RetriggerableDelay 3.0)
  (Variables|Default|SetReactivity%20Enabled false))

(event EventOnPostResIn
  (|CallParentFunction)
  (CallFunction|WeaponCheck (Variables|/Script/FortniteGame.FortPawn|GetCurrentWeapon (CustomCharacter|GetAssociatedPlayerPawn)))
  (if (CustomCosmeticModifier|IsinFrontEnd)
    ; Front End Loop
    (Utilities|Time|SetTimerbyEvent (delegate Custom|FrontEndReactivity) 1e+01 true)
    (Utilities|FlowControl|Delay 2.0)
    (else
      (bind (_statname _statvalue) (Cosmetics|WatchCosmeticStat (CustomCharacter|GetAssociatedPlayerPawn) "(TagName=\"StatManager.AthenaKills\")" 0 false)
        (:OnStatChanged
          (if (> _statvalue 0)
            _)))))
  ; FX
  (CallFunction|ReactivityEvent true)
  (Utilities|FlowControl|RetriggerableDelay 3.0)
  (Variables|Default|SetReactivity%20Enabled false))

(event EventOnWeaponChanged
  (params (NewWeapon :type Fort%20Weapon%20Object%20Reference) (OldWeapon :type Fort%20Weapon%20Object%20Reference))
  (CallFunction|WeaponCheck NewWeapon))

(event Custom|ReactivityEvent
  (params (isReactive? :type Boolean))
  (Effects|Components|ParticleSystem|SetFloatParameter (CallFunction|GetIdleVFXComponent) "GameTime" (Utilities|Time|GetGameTimeinSeconds))
  (Variables|Default|SetReactivity%20Enabled true))

(fn ConstructionScript
  (|CallParentFunction))

(fn WeaponCheck
  (params (NewWeapon :type Fort%20Weapon%20Object%20Reference))
  (bind _returnvalue (Visual|GetWeaponMesh NewWeapon))
  (bind _returnvalue_2 (Components|SkeletalMesh|GetSkeletalMeshAsset _returnvalue))
  (Utilities|IsValid NewWeapon
    (:Is%20Valid
      (Utilities|IsValid _returnvalue
        (:Is%20Valid
          (Utilities|IsValid _returnvalue_2
            (:Is%20Valid
              (Effects|Components|ParticleSystem|SetBoolParameter (CallFunction|GetIdleVFXComponent) "bIsWeaponOn" (Utilities|String|Contains (Utilities|GetObjectName _returnvalue_2) "OboeThorn_Axe")))))))))
