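; Backpack part modifier for the OceanBreeze cosmetic that hides its idle VFX while the pawn holds the OceanBreeze pickaxe.
; It compares the current weapon's skeletal mesh to OceanBreeze_Axe and shows the VFX component only when they differ.
; The check runs one tick after a weapon change and is duplicated in CheckWeapon, which a looping 0.1-second timer started on post-res-in polls.
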(event EventOnWeaponChanged
  (params (NewWeapon :type Fort%20Weapon%20Object%20Reference) (OldWeapon :type Fort%20Weapon%20Object%20Reference))
  (|CallParentFunction NewWeapon OldWeapon)
  (Utilities|FlowControl|DelayUntilNextTick)
  ;[1 1 1] Comment
  (Rendering|SetVisibility (CallFunction|GetIdleVFXComponent) (not (== (Variables|/Script/Engine.SkinnedMeshComponent|GetSkeletalMesh (Variables|/Script/FortniteGame.FortWeapon|GetWeaponMesh (Variables|/Script/FortniteGame.FortPawn|GetCurrentWeapon (CustomCharacter|GetAssociatedPlayerPawn)))) "/BRCosmetics/Weapons/FORT_Melee/PIckaxe_OceanBreeze/Meshes/OceanBreeze_Axe.OceanBreeze_Axe")) false))

(event EventOnPostResIn
  (|CallParentFunction)
  (Utilities|Time|SetTimerbyEvent (delegate Custom|CheckWeapon) 0.1 true))

(event Custom|CheckWeapon
  ;[1 1 1] Comment
  (Rendering|SetVisibility (CallFunction|GetIdleVFXComponent) (not (== (Variables|/Script/Engine.SkinnedMeshComponent|GetSkeletalMesh (Variables|/Script/FortniteGame.FortWeapon|GetWeaponMesh (Variables|/Script/FortniteGame.FortPawn|GetCurrentWeapon (CustomCharacter|GetAssociatedPlayerPawn)))) "/BRCosmetics/Weapons/FORT_Melee/PIckaxe_OceanBreeze/Meshes/OceanBreeze_Axe.OceanBreeze_Axe")) false))

(fn ConstructionScript
  (|CallParentFunction))
