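; Part modifier Blueprint for the Parcel Prank Surprise backpack cosmetic that plays its prank effects.
; On post-res-in, it casts the backpack part mesh's anim instance to the pack's AnimBP and binds to its PrankFX and PrankFXOff dispatchers.
; PrankFX spawns a smoke Niagara system and PrankFXOff spawns a close one, both attached at fxSocket on that mesh.
; The anim Blueprint drives effect timing, while this modifier only does the spawning.
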(event EventOnPostResIn
  (|CallParentFunction)
  (bind _asm_med_parcel_prank_surprise_pack_anim_bp (Utilities|Casting|CastToM_MED_Parcel_Prank_Surprise_Pack_AnimBP (Components|SkeletalMesh|GetAnimInstance (Customization|GetSkeletalMeshforPartType (CustomCharacter|GetAssociatedPlayerPawn) (CustomCharacter|GetCharacterPartType) :!class /Script/FortniteGame.FortPlayerPawn))))
  (Default|BindEventtoPrankFX _asm_med_parcel_prank_surprise_pack_anim_bp (EventDispatchers|CreateEvent :!event Custom|PrankFX))
  (Default|BindEventtoPrankFXOff _asm_med_parcel_prank_surprise_pack_anim_bp (EventDispatchers|CreateEvent :!event Custom|PrankFXOff)))

(event Custom|PrankFX
  (PartModifier|SpawnSystem|SpawnSystemAttachedPartModifier "/BRCosmetics/Accessories/FORT_Backpacks/M_MED_Parcel_Prank_Surprise_Backpack/FX/NS_ParcelPrankSurprise_Backpack_Smoke.NS_ParcelPrankSurprise_Backpack_Smoke" (Customization|GetSkeletalMeshforPartType (CustomCharacter|GetAssociatedPlayerPawn) (CustomCharacter|GetCharacterPartType) :!class /Script/FortniteGame.FortPlayerPawn) "fxSocket"))

(event Custom|PrankFXOff
  (PartModifier|SpawnSystem|SpawnSystemAttachedPartModifier "/BRCosmetics/Accessories/FORT_Backpacks/M_MED_Parcel_Prank_Surprise_Backpack/FX/NS_ParcelPrankSurprise_Backpack_Close.NS_ParcelPrankSurprise_Backpack_Close" (Customization|GetSkeletalMeshforPartType (CustomCharacter|GetAssociatedPlayerPawn) (CustomCharacter|GetCharacterPartType) :!class /Script/FortniteGame.FortPlayerPawn) "fxSocket"))

(fn ConstructionScript
  (|CallParentFunction))
