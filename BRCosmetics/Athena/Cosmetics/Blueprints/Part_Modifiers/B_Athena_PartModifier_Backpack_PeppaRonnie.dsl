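; Part modifier Blueprint for the PeppaRonnie backpack cosmetic that adds visual effects when the wearer jumps and lands.
; The pawn jumped and landed events each call the parent, then spawn their own jump or land Niagara system.
; Each system attaches to the root of the skeletal mesh retrieved for the custom character part on the player pawn.
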(event EventOnPawnJumped
  (|CallParentFunction)
  (PartModifier|SpawnSystem|SpawnSystemAttachedPartModifier "/BRCosmetics/Accessories/FORT_Backpacks/Backpack_M_MED_PeppaRonnie/VFX/NS_Backpack_PeppaRonnie_Jump.NS_Backpack_PeppaRonnie_Jump" (CustomCharacter|GetSkeletalMeshForPartOnPlayerPawn) "root"))

(event EventOnPawnLanded
  (|CallParentFunction)
  (PartModifier|SpawnSystem|SpawnSystemAttachedPartModifier "/BRCosmetics/Accessories/FORT_Backpacks/Backpack_M_MED_PeppaRonnie/VFX/NS_Backpack_PeppaRonnie_Land.NS_Backpack_PeppaRonnie_Land" (CustomCharacter|GetSkeletalMeshForPartOnPlayerPawn) "root"))

(fn ConstructionScript)
