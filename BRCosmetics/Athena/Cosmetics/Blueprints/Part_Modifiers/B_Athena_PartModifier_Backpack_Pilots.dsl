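; Cosmetic part modifier for the Pilots backpack that shows the player's Athena kill count on a numbers mesh attached to the backpack.
; On pawn or part change it reads the kill stat and watches it, updating material parameters and playing a timeline-driven FlashControl flash on each change.
; UpdateMaterialParameters splits the count into tens and ones scalars with a single-digit switch, leaving the display unchanged above 99.
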(event Collision|EventActorBeginOverlap
  (params (OtherActor :type Actor%20Object%20Reference)))

(event EventTick
  (params (DeltaSeconds :type Float%20%28single-precision%29)))

(event EventOnPawnAndPartChanged
  (|CallParentFunction)
  (Transformation|AttachComponentToComponent (Variables|Default|GetM_MED_Pilots_Backpack_Numbers) (CustomCharacter|GetSkeletalMeshForPartOnPlayerPawn) "backpack_root" "KeepRelative" "KeepRelative" "KeepRelative" true)
  (bind _returnvalue (Class|FortCosmeticFunctionLibrary|GetCosmeticStatValue (CustomCharacter|GetAssociatedPlayerPawn) "(TagName=\"StatManager.AthenaKills\")"))
  (Variables|Default|SetKillCount (Math|Conversions|ToFloat(Integer) _returnvalue))
  (CallFunction|UpdateMaterialParameters)
  (bind (_statname _statvalue) (Cosmetics|WatchCosmeticStat (CustomCharacter|GetAssociatedPlayerPawn) "(TagName=\"StatManager.AthenaKills\")" 0 false)
    (:OnStatChanged
      (Variables|Default|SetKillCount (Math|Conversions|ToFloat(Integer) _statvalue))
      (CallFunction|UpdateMaterialParameters)
      (bind (_direction _lightconrol) (Timeline|UpdateNumber PlayFromStart)
        (:Update
          (Rendering|Material|SetScalarParameterValueonMaterials (Variables|Default|GetM_MED_Pilots_Backpack_Numbers) "FlashControl" _lightconrol))
        (:Finished
          (Rendering|Material|SetScalarParameterValueonMaterials (Variables|Default|GetM_MED_Pilots_Backpack_Numbers) "FlashControl" 0.0))))))

(event EventOnPostResIn
  (|CallParentFunction)
  (CallFunction|UpdateMaterialParameters))

(fn UpdateMaterialParameters
  (bind _killcount (Variables|Default|GetKillCount))
  (bind _returnvalue (/ _killcount 1e+01))
  (bind _returnvalue_2 (Math|Float|Fraction _returnvalue))
  (bind _returnvalue_3 (* _returnvalue_2 1e+01))
  (bind _killcount_2 (Variables|Default|GetKillCount))
  (bind _m_med_pilots_backpack_numbers (Variables|Default|GetM_MED_Pilots_Backpack_Numbers))
  ;[1 1 1] Simple, 2 digit kill counter designed for material use
  (if (> _killcount_2 0.0)
    (if (> _killcount_2 9.0)
      (if (not (> _killcount_2 99.0))
        (Rendering|Material|SetScalarParameterValueonMaterials _m_med_pilots_backpack_numbers "UseSingleDigit" 0.0)
        (Rendering|Material|SetScalarParameterValueonMaterials _m_med_pilots_backpack_numbers "Kills-Tens" (Math|Conversions|ToFloat(Integer) (Math|Float|Floor _returnvalue)))
        (Rendering|Material|SetScalarParameterValueonMaterials _m_med_pilots_backpack_numbers "Kills-Ones" _returnvalue_3))
      (else
        ;;[1 1 1] Use single digit
        (Rendering|Material|SetScalarParameterValueonMaterials _m_med_pilots_backpack_numbers "UseSingleDigit" 1.0)
        (Rendering|Material|SetScalarParameterValueonMaterials _m_med_pilots_backpack_numbers "Kills-Tens" 0.0)
        (Rendering|Material|SetScalarParameterValueonMaterials _m_med_pilots_backpack_numbers "Kills-Ones" _returnvalue_3)))
    (else
      (Rendering|Material|SetScalarParameterValueonMaterials _m_med_pilots_backpack_numbers "UseSingleDigit" 1.0)
      (Rendering|Material|SetScalarParameterValueonMaterials _m_med_pilots_backpack_numbers "Kills-Tens" 0.0)
      (Rendering|Material|SetScalarParameterValueonMaterials _m_med_pilots_backpack_numbers "Kills-Ones" 0.0))))

(fn ConstructionScript)
