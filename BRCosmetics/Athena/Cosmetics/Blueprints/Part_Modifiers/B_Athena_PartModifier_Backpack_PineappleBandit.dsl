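; Cosmetic part modifier for the Pineapple Bandit backpack that drives its idle particle effect from player movement.
; The effect is deactivated in the front end after resurrect-in.
; On movement updates, VelocityCheck enables the Trail emitter when prior velocity is non-zero without current vertical velocity, and disables it during vertical motion.
; A delayed jump check hides the trail, briefly pulses JumpSpawnRate, then restores the trail.
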(event EventOnPostResIn
  (|CallParentFunction)
  (if (CustomCosmeticModifier|IsinFrontEnd)
    (Components|Activation|Deactivate (CallFunction|GetIdleVFXComponent))))

(event EventOnCharacterMovementUpdated
  (params (DeltaSeconds :type Float%20%28single-precision%29) (OldLocation :type Vector :default "0, 0, 0") (OldVelocity :type Vector :default "0, 0, 0"))
  (bind (_movement_output _z_velocity_output) (CallFunction|VelocityCheck OldVelocity))
  (bind _returnvalue (CallFunction|GetIdleVFXComponent))
  (bind _returnvalue_2 (CallFunction|GetIdleVFXComponent))
  (|CallParentFunction DeltaSeconds OldLocation OldVelocity)
  (Utilities|FlowControl|Sequence
    (:then_0
      (if _movement_output
        (Effects|Components|ParticleSystem|SetEmitterEnable _returnvalue_2 "Trail" true)
        (elif _z_velocity_output
          (Effects|Components|ParticleSystem|SetEmitterEnable _returnvalue_2 "Trail" false))))
    (:then_1
      (Utilities|FlowControl|Delay 0.15)
      (if (Character|IsJumpProvidingForce (CustomCharacter|GetAssociatedPlayerPawn))
        (Effects|Components|ParticleSystem|SetEmitterEnable _returnvalue "Trail" false)
        (Effects|Components|ParticleSystem|SetFloatParameter _returnvalue "JumpSpawnRate" 1.0)
        (Utilities|FlowControl|Delay 0.07)
        (Effects|Components|ParticleSystem|SetFloatParameter _returnvalue "JumpSpawnRate" 0.0)
        (Utilities|FlowControl|Delay 0.6)
        (Effects|Components|ParticleSystem|SetEmitterEnable _returnvalue "Trail" true)))))

(fn ConstructionScript)

(fn VelocityCheck
  (params (InVec :type Vector))
  (return (and (Math|Boolean|ORBoolean (!= (Math|Float|Absolute(Float) (.x InVec)) 0.0) (!= (Math|Float|Absolute(Float) (.y InVec)) 0.0) (!= (Math|Float|Absolute(Float) (.z InVec)) 0.0)) (== (.z (Transformation|GetVelocity (CustomCharacter|GetAssociatedPlayerPawn))) 0.0)) (!= (.z (Transformation|GetVelocity (CustomCharacter|GetAssociatedPlayerPawn))) 0.0)))
