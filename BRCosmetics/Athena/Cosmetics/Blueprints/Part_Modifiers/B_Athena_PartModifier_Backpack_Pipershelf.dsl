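; Backpack cosmetic part modifier for Pipershelf that adjusts the back bling's glow and idle VFX.
; After resing in, it raises the diamond emissive intensity material parameter on the part's mesh, only in the front end.
; On weapon change, it waits briefly, then checks the part's anim instance for a pickaxe with the PiperShelf meta tag.
; The result sets the idle particle system's bPickaxeEquipped parameter, letting the effect react to the matching pickaxe.
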(event EventOnPostResIn
  (|CallParentFunction)
  (if (CustomCosmeticModifier|IsinFrontEnd)
    (Rendering|Material|SetScalarParameterValueonMaterials (CustomCharacter|GetSkeletalMeshForPartOnPlayerPawn) "DiamondEmissiveIntensity" 3.6)))

(event EventOnWeaponChanged
  (params (NewWeapon :type Fort%20Weapon%20Object%20Reference) (OldWeapon :type Fort%20Weapon%20Object%20Reference))
  (bind _returnvalue (CallFunction|GetIdleVFXComponent))
  (|CallParentFunction NewWeapon OldWeapon)
  (Utilities|FlowControl|Delay 0.1)
  (bind _ascustom_character_part_anim_instance (Utilities|Casting|CastToCustomCharacterPartAnimInstance (Components|SkeletalMesh|GetAnimInstance (Customization|GetSkeletalMeshforPartType (CustomCharacter|GetAssociatedPlayerPawn) (CustomCharacter|GetCharacterPartType) :!class /Script/FortniteGame.FortPlayerPawn))))
  (if (CosmeticTags|PickaxeItemDefinitionHasAnyMetaTag _ascustom_character_part_anim_instance (Utilities|Array|MakeArray "(TagName=\"Cosmetics.MetaTags.PiperShelf\")") true)
    (Effects|Components|ParticleSystem|SetBoolParameter _returnvalue "bPickaxeEquipped" true)
    (else
      (Effects|Components|ParticleSystem|SetBoolParameter _returnvalue "bPickaxeEquipped" false))))

(fn ConstructionScript
  (|CallParentFunction))
