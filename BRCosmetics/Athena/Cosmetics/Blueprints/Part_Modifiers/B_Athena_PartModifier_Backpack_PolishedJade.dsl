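; Part modifier for the PolishedJade back bling that makes it react to player kills across three visual stages.
; Stage activation events, or the kill stat reverse-matched against thresholds on pawn change, set the current stage.
; Each stage toggles idle VFX bool parameters and sets material mask radii, with a timeline fading in the final stage blend.
; In frontend reactivity display, a timeline started after res-in plays through all three stages as a preview.
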(event EventOnCosmeticStateActivated
  (params (Stat%20Name :type Name :default "None") (Stat%20Value :type Integer) (STAGE%20Num%20Being%20Activated :type Integer))
  (|CallParentFunction Stat%20Name Stat%20Value STAGE%20Num%20Being%20Activated)
  (Variables|Default|SetCurrentStage STAGE%20Num%20Being%20Activated)
  (CallFunction|UpdateMaterialParameters)
  ; material blend
  (switch int (Variables|Default|GetCurrentStage)
    (:0
      (Rendering|Material|SetScalarParameterValueonMaterials (CustomCharacter|GetSkeletalMeshForPartOnPlayerPawn) "EffectMaskRadius" 0.0)
      (Rendering|Material|SetScalarParameterValueonMaterials (CustomCharacter|GetSkeletalMeshForPartOnPlayerPawn) "stage2Blend" 0.0)
      (Rendering|Material|SetScalarParameterValueonMaterials (CustomCharacter|GetSkeletalMeshForPartOnPlayerPawn) "EffectMaskRadius_Stage2" 0.0))
    (:1
      (Rendering|Material|SetScalarParameterValueonMaterials (CustomCharacter|GetSkeletalMeshForPartOnPlayerPawn) "EffectMaskRadius" 0.1)
      (Rendering|Material|SetScalarParameterValueonMaterials (CustomCharacter|GetSkeletalMeshForPartOnPlayerPawn) "stage2Blend" 0.0)
      (Rendering|Material|SetScalarParameterValueonMaterials (CustomCharacter|GetSkeletalMeshForPartOnPlayerPawn) "EffectMaskRadius_Stage2" 0.0))
    (:2
      (Rendering|Material|SetScalarParameterValueonMaterials (CustomCharacter|GetSkeletalMeshForPartOnPlayerPawn) "EffectMaskRadius_Stage2" 1.0)
      (bind (_direction _stage2_blend) (Timeline|Timeline PlayFromStart)
        (:Update
          (Rendering|Material|SetScalarParameterValueonMaterials (CustomCharacter|GetSkeletalMeshForPartOnPlayerPawn) "stage2Blend" _stage2_blend))
        (:Finished
          (Rendering|Material|SetScalarParameterValueonMaterials (CustomCharacter|GetSkeletalMeshForPartOnPlayerPawn) "stage2Blend" 1.0))))))

(event EventOnPawnAndPartChanged
  (|CallParentFunction)
  (if (not (Reactive|Frontend|ShouldDisplayFrontendReactivity 0 "Backpack"))
    (bind _returnvalue (Class|FortCosmeticFunctionLibrary|GetCosmeticStatValue (CustomCharacter|GetAssociatedPlayerPawn) "(TagName=\"StatManager.AthenaKills\")"))
    (bind _stage (CallFunction|StatValuetoStage _returnvalue))
    (Variables|Default|SetCurrentStage _stage)
    (CallFunction|UpdateMaterialParameters)
    ; material blend
    (switch int (Variables|Default|GetCurrentStage)
      (:0
        (Rendering|Material|SetScalarParameterValueonMaterials (CustomCharacter|GetSkeletalMeshForPartOnPlayerPawn) "EffectMaskRadius" 0.0)
        (Rendering|Material|SetScalarParameterValueonMaterials (CustomCharacter|GetSkeletalMeshForPartOnPlayerPawn) "stage2Blend" 0.0)
        (Rendering|Material|SetScalarParameterValueonMaterials (CustomCharacter|GetSkeletalMeshForPartOnPlayerPawn) "EffectMaskRadius_Stage2" 0.0))
      (:1
        (Rendering|Material|SetScalarParameterValueonMaterials (CustomCharacter|GetSkeletalMeshForPartOnPlayerPawn) "EffectMaskRadius" 0.1)
        (Rendering|Material|SetScalarParameterValueonMaterials (CustomCharacter|GetSkeletalMeshForPartOnPlayerPawn) "stage2Blend" 0.0)
        (Rendering|Material|SetScalarParameterValueonMaterials (CustomCharacter|GetSkeletalMeshForPartOnPlayerPawn) "EffectMaskRadius_Stage2" 0.0))
      (:2
        (Rendering|Material|SetScalarParameterValueonMaterials (CustomCharacter|GetSkeletalMeshForPartOnPlayerPawn) "EffectMaskRadius_Stage2" 1.0)
        (Timeline|Timeline PlayFromStart)))))

(event EventOnPostResIn
  (|CallParentFunction)
  (if (Reactive|Frontend|ShouldDisplayFrontendReactivity 0 "Backpack")
    (Timeline|Timeline_0 PlayFromStart
      (:Stage%201
        (CallFunction|StageOne)
        (Rendering|Material|SetScalarParameterValueonMaterials (CustomCharacter|GetSkeletalMeshForPartOnPlayerPawn) "EffectMaskRadius" 0.0)
        (Rendering|Material|SetScalarParameterValueonMaterials (CustomCharacter|GetSkeletalMeshForPartOnPlayerPawn) "stage2Blend" 0.0)
        (Rendering|Material|SetScalarParameterValueonMaterials (CustomCharacter|GetSkeletalMeshForPartOnPlayerPawn) "EffectMaskRadius_Stage2" 0.0))
      (:Stage%202
        (CallFunction|StageTwo)
        (Rendering|Material|SetScalarParameterValueonMaterials (CustomCharacter|GetSkeletalMeshForPartOnPlayerPawn) "EffectMaskRadius" 0.1)
        (Rendering|Material|SetScalarParameterValueonMaterials (CustomCharacter|GetSkeletalMeshForPartOnPlayerPawn) "stage2Blend" 0.0)
        (Rendering|Material|SetScalarParameterValueonMaterials (CustomCharacter|GetSkeletalMeshForPartOnPlayerPawn) "EffectMaskRadius_Stage2" 0.0))
      (:Stage%203
        (CallFunction|StageThree)
        (Rendering|Material|SetScalarParameterValueonMaterials (CustomCharacter|GetSkeletalMeshForPartOnPlayerPawn) "EffectMaskRadius" 1.0)
        (bind (_direction _stage2_blend) (Timeline|Timeline_1 PlayFromStart)
          (:Update
            (Rendering|Material|SetScalarParameterValueonMaterials (CustomCharacter|GetSkeletalMeshForPartOnPlayerPawn) "stage2Blend" _stage2_blend))
          (:Finished
            (Rendering|Material|SetScalarParameterValueonMaterials (CustomCharacter|GetSkeletalMeshForPartOnPlayerPawn) "stage2Blend" 1.0)))))))

(event Custom|StageOne
  (bind _returnvalue (CallFunction|GetIdleVFXComponent))
  (Effects|Components|ParticleSystem|SetBoolParameter _returnvalue "Stage_01" false)
  (Effects|Components|ParticleSystem|SetBoolParameter _returnvalue "Stage_02" false))

(event Custom|StageTwo
  (bind _returnvalue (CallFunction|GetIdleVFXComponent))
  (Effects|Components|ParticleSystem|SetBoolParameter _returnvalue "Stage_01" true)
  (Effects|Components|ParticleSystem|SetBoolParameter _returnvalue "Stage_02" false))

(event Custom|StageThree
  (bind _returnvalue (CallFunction|GetIdleVFXComponent))
  (Effects|Components|ParticleSystem|SetBoolParameter _returnvalue "Stage_01" true)
  (Effects|Components|ParticleSystem|SetBoolParameter _returnvalue "Stage_02" true))

(fn StatValueToStage
  (params (StatValue :type Integer))
  (bind (_arrayindex _arrayelement) (Utilities|Array|ReverseforEachLoop (Variables|PartModifierControls|WatchCosmeticStat|GetStat%20Thresholds))
    (:LoopBody
      (if (>= (Variables|GetStatValue) _arrayelement)
        (return _arrayindex)))
    (:Completed
      (return 0))))

(fn UpdateMaterialParameters
  (switch int (Variables|Default|GetCurrentStage)
    (:0
      (CallFunction|StageOne))
    (:1
      (CallFunction|StageTwo))
    (:2
      (CallFunction|StageThree))))

(fn ConstructionScript
  (|CallParentFunction))
