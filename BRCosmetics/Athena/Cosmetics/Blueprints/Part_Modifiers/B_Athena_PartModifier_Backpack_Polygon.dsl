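; Part modifier for the Polygon back bling, which steps through four reactive looks based on the player's kill count.
; On init it creates the backpack MID and applies the current stage; a looping timer cycles stages for frontend previews.
; A kill-stat watcher activates idle VFX, plays per-stage sounds and calls stage events.
; These set SuperLevel on the backpack MID, with upper stages time-stamping SuperLevelAnim on FaceMID, which is never assigned here.
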(event EventOnPostResIn
  (|CallParentFunction)
  (Utilities|FlowControl|Sequence
    (:then_0
      ;[1 1 1] MID Setup
      (bind _returnvalue (Rendering|Material|CreateDynamicMaterialInstance (CustomCharacter|GetSkeletalMeshForPartOnPlayerPawn) 0 (Rendering|Material|GetMaterial (CustomCharacter|GetSkeletalMeshForPartOnPlayerPawn) 0) "Polygon_Body_MID" :!class /Script/Engine.PrimitiveComponent))
      (Variables|Default|SetBackpackMID _returnvalue)
      (CallFunction|SetMaterial))
    (:then_1
      ;[1 1 1] Front End Preview
      (if (Reactive|Frontend|ShouldDisplayFrontendReactivity 0 "Backpack")
        (Utilities|Time|SetTimerbyEvent (EventDispatchers|CreateEvent :!event Custom|FrontEndTrigger) 3.0 true)))))

(event EventOnPawnAndPartChanged
  ;[1 1 1] VFX Spawn
  (|CallParentFunction)
  (bind (_statname _statvalue) (Cosmetics|WatchCosmeticStat (CustomCharacter|GetAssociatedPlayerPawn) "(TagName=\"StatManager.AthenaKills\")" 0 false)
    (:OnStatChanged
      (Components|Activation|Activate (CallFunction|GetIdleVFXComponent) true)
      (Development|PrintString (Utilities|String|ToString(Integer) (Math|Integer|Clamp(Integer) _statvalue 0 3)))
      (switch int (Math|Integer|Clamp(Integer) _statvalue 0 3)
        (:0
          (CallFunction|Stage0))
        (:1
          ;[1 0.3041 0.07268] Play Sounds
          (CallFunction|PlayBackBlingAudio "/BRCosmetics/Sounds/Backblings/PolygonMale/Backbling_PolygonMale_01_Cue.Backbling_PolygonMale_01_Cue")
          (CallFunction|Stage1))
        (:2
          (CallFunction|PlayBackBlingAudio "/BRCosmetics/Sounds/Backblings/PolygonMale/Backbling_PolygonMale_02_Cue.Backbling_PolygonMale_02_Cue")
          (CallFunction|Stage2))
        (:3
          (CallFunction|PlayBackBlingAudio "/BRCosmetics/Sounds/Backblings/PolygonMale/Backbling_PolygonMale_03_Cue.Backbling_PolygonMale_03_Cue")
          (CallFunction|Stage3))))))

(event Custom|FrontEndTrigger
  ;[1 1 1] Front End Preview
  (switch int (Variables|Default|GetFrontEndSwitch)
    (:0
      (CallFunction|Stage0))
    (:1
      (CallFunction|Stage1))
    (:2
      (CallFunction|Stage2))
    (:3
      (CallFunction|Stage3)
      (Variables|Default|SetFrontEndSwitch 0)))
  (Math|Integer|IncrementInt (Variables|Default|GetFrontEndSwitch)))

(event Custom|SetMaterial
  ;[1 1 1] SetMaterials
  (bind _returnvalue (Class|FortCosmeticFunctionLibrary|GetCosmeticStatValue (CustomCharacter|GetAssociatedPlayerPawn) "(TagName=\"StatManager.AthenaKills\")"))
  (switch int (Math|Integer|Clamp(Integer) _returnvalue 0 3)
    (:0
      (CallFunction|Stage0))
    (:1
      (CallFunction|Stage1))
    (:2
      (CallFunction|Stage2))
    (:3
      (CallFunction|Stage3))))

(event EventOnCosmeticStateActivated
  (params (Stat%20Name :type Name :default "None") (Stat%20Value :type Integer) (STAGE%20Num%20Being%20Activated :type Integer))
  (|CallParentFunction Stat%20Name Stat%20Value STAGE%20Num%20Being%20Activated))

(event Custom|Stage%200
  ;[1 1 1] Material and VFX Setup for Stages
  (Rendering|Material|SetScalarParameterValue (Variables|Default|GetBackpackMID) "SuperLevel" 0.0 :!class /Script/Engine.MaterialInstanceDynamic))

(event Custom|Stage%201
  ;[1 1 1] Material and VFX Setup for Stages
  (Rendering|Material|SetScalarParameterValue (Variables|Default|GetBackpackMID) "SuperLevel" 1.0 :!class /Script/Engine.MaterialInstanceDynamic)
  (Rendering|Material|SetScalarParameterValue (Variables|Default|GetFaceMID) "SuperLevelAnim" (Utilities|Time|GetGameTimeinSeconds) :!class /Script/Engine.MaterialInstanceDynamic))

(event Custom|Stage%202
  ;[1 1 1] Material and VFX Setup for Stages
  (Rendering|Material|SetScalarParameterValue (Variables|Default|GetBackpackMID) "SuperLevel" 2.0 :!class /Script/Engine.MaterialInstanceDynamic)
  (Rendering|Material|SetScalarParameterValue (Variables|Default|GetFaceMID) "SuperLevelAnim" (Utilities|Time|GetGameTimeinSeconds) :!class /Script/Engine.MaterialInstanceDynamic))

(event Custom|Stage%203
  ;[1 1 1] Material and VFX Setup for Stages
  (Rendering|Material|SetScalarParameterValue (Variables|Default|GetBackpackMID) "SuperLevel" 3.0 :!class /Script/Engine.MaterialInstanceDynamic)
  (Rendering|Material|SetScalarParameterValue (Variables|Default|GetFaceMID) "SuperLevelAnim" (Utilities|Time|GetGameTimeinSeconds) :!class /Script/Engine.MaterialInstanceDynamic))

(fn PlayBackBlingAudio
  (params (Sound :type Sound%20Base%20Object%20Reference))
  (Audio|SpawnSoundAttached Sound (Customization|GetSkeletalMeshforPartType (CustomCharacter|GetAssociatedPlayerPawn) "Backpack" :!class /Script/FortniteGame.FortPlayerPawn)))

(fn ConstructionScript
  (|CallParentFunction))
