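; Part modifier for the PrairieGizmo back bling that makes it react to the owning player's eliminations.
; OnKill switches the idle Niagara component's ReactivityEnabled bool on, then off after three seconds.
; In game a kill-stat watcher triggers it and an attached Apprentice elimination sound plays.
; In frontend display, a delayed first call and a looping timer demo the effect silently.
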(event EventOnPostResIn
  (|CallParentFunction)
  (Variables|Default|SetReactivityEnabled false)
  (if (Reactive|Frontend|IsCosmeticObjectDisplayedinFrontendContext 0)
    (Utilities|FlowControl|Delay 4.0)
    (CallFunction|OnKill)
    (Utilities|Time|SetTimerbyEvent (delegate Custom|OnKill) 13.0 true)
    (else
      ;[1 1 1] Setup On Kill
      (Cosmetics|WatchCosmeticStat (CustomCharacter|GetAssociatedPlayerPawn) "(TagName=\"StatManager.AthenaKills\")" 0 false
        (:OnStatChanged
          (CallFunction|OnKill))))))

(event Custom|OnKill
  ;[1 1 1] On Kill Event
  (bind _output_get (Variables|Default|SetReactivityEnabled true))
  (bind _asniagara_particle_system_component (Utilities|Casting|CastToNiagaraComponent (CallFunction|GetIdleVFXComponent)))
  (Niagara|SetNiagaraVariable(Bool) _asniagara_particle_system_component "ReactivityEnabled" _output_get)
  (if (not (Reactive|Frontend|IsCosmeticObjectDisplayedinFrontendContext 0))
    (Default|SetandPlaySound))
  (Utilities|FlowControl|Delay 3.0)
  (bind _output_get_2 (Variables|Default|SetReactivityEnabled false))
  (Niagara|SetNiagaraVariable(Bool) _asniagara_particle_system_component "ReactivityEnabled" _output_get_2))

(fn SetAndPlaySound
  (Audio|SpawnSoundAttached "/BRCosmetics/Sounds/Backblings/Apprentice/BackBling_Apprentice_Elimination_Open_Cue.BackBling_Apprentice_Elimination_Open_Cue" (CustomCharacter|GetSkeletalMeshForPartOnPlayerPawn)))

(fn ConstructionScript
  (|CallParentFunction))
