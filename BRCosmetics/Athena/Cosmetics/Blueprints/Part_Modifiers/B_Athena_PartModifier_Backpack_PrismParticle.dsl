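; Cosmetic part modifier that drives the PrismParticle backpack's dynamic material, making it reactive to damage dealt.
; Damage stat changes stamp game time into the material; in the front end, a preview timeline event does the same.
; Outside the front end, a post-update-work tick sends the dyn_pack_socket position and remapped basis axes to the material.
; When the socket is over 1000 units from the pawn, the material switches to object position instead.
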(event EventOnPostResIn
  ;[1 1 1] FE cycles
  (|CallParentFunction)
  (if (CustomCosmeticModifier|IsinFrontEnd)
    (Timeline|PrismParticle_FE_Preview PlayFromStart
      (:FE_Reactive
        ;[1 1 1] Reactivity
        (Rendering|Material|SetScalarParameterValue (Class|BTechArtGlobalLibrary|GetDynamicMaterialInstance (CustomCharacter|GetSkeletalMeshForPartOnPlayerPawn)) "PP_GameTimeFromBP_CCPM" (Utilities|Time|GetGameTimeinSeconds) :!class /Script/Engine.MaterialInstanceDynamic)))))

(event EventOnPawnAndPartChanged
  ;[1 1 1] Reactivity
  (|CallParentFunction)
  (Cosmetics|WatchCosmeticStat (CustomCharacter|GetAssociatedPlayerPawn) "(TagName=\"StatManager.DamageDealt\")" 0 false
    (:OnStatChanged
      (Rendering|Material|SetScalarParameterValue (Class|BTechArtGlobalLibrary|GetDynamicMaterialInstance (CustomCharacter|GetSkeletalMeshForPartOnPlayerPawn)) "PP_GameTimeFromBP_CCPM" (Utilities|Time|GetGameTimeinSeconds) :!class /Script/Engine.MaterialInstanceDynamic))))

(event EventBeginPlay
  ;[1 1 1] Send location and rotation data to material
  (Actor|Tick|SetTickGroup "TG_PostUpdateWork"))

(event EventTick
  (params (DeltaSeconds :type Float%20%28single-precision%29))
  (bind _returnvalue (CustomCharacter|GetSkeletalMeshForPartOnPlayerPawn))
  (bind _returnvalue_2 (Transformation|GetSocketTransform _returnvalue "dyn_pack_socket" "RTS_World"))
  (bind _location (.location _returnvalue_2))
  (bind _rotation (.rotation _returnvalue_2))
  (bind _material_instance_dynamic (Class|BTechArtGlobalLibrary|GetDynamicMaterialInstance _returnvalue))
  ;[1 1 1] Send location and rotation data to material
  (if (CustomCosmeticModifier|IsinFrontEnd)
    _
    (elif (> (Math|Vector|Distance(Vector) (Transformation|GetActorLocation (CustomCharacter|GetAssociatedPlayerPawn)) _location) 1e+03)
      (Rendering|Material|SetScalarParameterValue _material_instance_dynamic "PP_UseObjectPosition" 1.0 :!class /Script/Engine.MaterialInstanceDynamic)
      (else
        (Rendering|Material|SetScalarParameterValue _material_instance_dynamic "PP_UseObjectPosition" 0.0 :!class /Script/Engine.MaterialInstanceDynamic)
        (Rendering|Material|SetVectorParameterValue _material_instance_dynamic "PP_PivotLocation" (Math|Conversions|ToLinearColor(Vector) _location) :!class /Script/Engine.MaterialInstanceDynamic)
        (Rendering|Material|SetVectorParameterValue _material_instance_dynamic "PP_PivotBasisX" (Math|Conversions|ToLinearColor(Vector) (Math|Vector|GetUpVector _rotation)) :!class /Script/Engine.MaterialInstanceDynamic)
        (Rendering|Material|SetVectorParameterValue _material_instance_dynamic "PP_PivotBasisY" (Math|Conversions|ToLinearColor(Vector) (Math|Vector|GetRightVector _rotation)) :!class /Script/Engine.MaterialInstanceDynamic)
        (Rendering|Material|SetVectorParameterValue _material_instance_dynamic "PP_PivotBasisZ" (Math|Conversions|ToLinearColor(Vector) (* (Math|Vector|GetForwardVector _rotation) -1.0)) :!class /Script/Engine.MaterialInstanceDynamic)))))

(fn ConstructionScript
  (|CallParentFunction))
