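; Backpack cosmetic part modifier for the PunkDevil Summer back bling that reacts when a tracked stat changes, noted as eliminations.
; On a stat change it calls the parent handler, then gets the skeletal mesh for the part on the player pawn.
; It writes the current game time and a fixed reaction length into material scalar parameters, leaving the timed reactive effect to the material.
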(event EventOnStatChanged
  (params (State%20Name :type Name :default "None") (Stat%20Value :type Integer))
  (bind _returnvalue (CustomCharacter|GetSkeletalMeshForPartOnPlayerPawn))
  ;[1 1 1] Elimination
  (|CallParentFunction State%20Name Stat%20Value)
  (Rendering|Material|SetScalarParameterValueonMaterials _returnvalue "PD_GameTimeFromBP" (Utilities|Time|GetGameTimeinSeconds))
  (Rendering|Material|SetScalarParameterValueonMaterials _returnvalue "PD_ReactiveLength" 3.0))

(fn ConstructionScript)
