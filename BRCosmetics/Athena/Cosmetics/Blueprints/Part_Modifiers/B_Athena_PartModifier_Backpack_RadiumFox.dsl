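; Part modifier for the RadiumFox backpack cosmetic that shows reactive emote states through material parameters on the part's mesh.
; In a match, cosmetic stat watchers for damage, kills, healing and chest opens set ReactivityStatus plus start times or kill count, then reset it after retriggerable delays.
; Boolean flags prioritize damage, then kills, healing and chests.
; In the front end, a timeline drives status and damage progress instead.
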(event Custom|DamageRecivedEmote
  (bind _returnvalue (CustomCharacter|GetSkeletalMeshForPartOnPlayerPawn))
  (Variables|Default|SetIsDamageReceivedEmote true)
  (Rendering|Material|SetScalarParameterValueonMaterials (CustomCharacter|GetSkeletalMeshForPartOnPlayerPawn) "ReactivityStatus" 3.0)
  (Rendering|Material|SetScalarParameterValueonMaterials _returnvalue "DamageReceivedStartTime" (Utilities|Time|GetGameTimeinSeconds))
  (Utilities|FlowControl|RetriggerableDelay (Variables|Default|GetEmoteDuration))
  (Rendering|Material|SetScalarParameterValueonMaterials _returnvalue "ReactivityStatus" 0.0)
  (Variables|Default|SetIsDamageReceivedEmote false))

(event EventOnPostResIn
  (bind _returnvalue (CustomCharacter|GetAssociatedPlayerPawn))
  (bind _returnvalue_2 (CustomCharacter|GetSkeletalMeshForPartOnPlayerPawn))
  (|CallParentFunction)
  (if (CustomCosmeticModifier|IsinFrontEnd)
    (Rendering|Material|SetScalarParameterValueonMaterials (CustomCharacter|GetSkeletalMeshForPartOnPlayerPawn) "IsFrontEnd" 1.0)
    (bind (_direction _status _damageprogress) (Timeline|StatusTimeLine Play)
      (:Update
        (Rendering|Material|SetScalarParameterValueonMaterials _returnvalue_2 "ReactivityStatus" _status)
        (Rendering|Material|SetScalarParameterValueonMaterials _returnvalue_2 "DamageReceived_FrontEnd" _damageprogress)))
    (else
      (Rendering|Material|SetScalarParameterValueonMaterials (CustomCharacter|GetSkeletalMeshForPartOnPlayerPawn) "ReactivityStatus" 0.0)
      (Variables|Default|SetCurrentHealth (FortPawn|GetHealth (CustomCharacter|GetAssociatedPlayerPawn)))
      (Utilities|FlowControl|Sequence
        (:then_0
          (Cosmetics|WatchCosmeticStat _returnvalue "(TagName=\"StatManager.DamageReceived\")" 0 false
            (:OnStatChanged
              (CallFunction|DamageRecivedEmote))))
        (:then_1
          (bind (_statname _statvalue) (Cosmetics|WatchCosmeticStat _returnvalue "(TagName=\"StatManager.AthenaKills\")" 0 false)
            (:OnStatChanged
              (CallFunction|AthenaKillEmote _statvalue))))
        (:then_2
          (Cosmetics|WatchCosmeticStat _returnvalue "(TagName=\"StatManager.GainedHealthTimes\")" 0 false
            (:OnStatChanged
              (CallFunction|HealingEmote))))
        (:then_3
          (Cosmetics|WatchCosmeticStat _returnvalue "(TagName=\"StatManager.OpenChests\")" 0 false
            (:OnStatChanged
              (CallFunction|OpenChestEmote))))))))

(event Custom|AthenaKillEmote
  (params (KillAmount :type Integer))
  (if (not (Variables|Default|GetIsDamageReceivedEmote))
    (Variables|Default|SetIsAthenaKillEmote true)
    (Rendering|Material|SetScalarParameterValueonMaterials (CustomCharacter|GetSkeletalMeshForPartOnPlayerPawn) "ReactivityStatus" 1.0)
    (Rendering|Material|SetScalarParameterValueonMaterials (CustomCharacter|GetSkeletalMeshForPartOnPlayerPawn) "ElimationNumberInt" (Math|Conversions|ToFloat(Integer) KillAmount))
    (Utilities|FlowControl|RetriggerableDelay (Variables|Default|GetEmoteDuration))
    (Rendering|Material|SetScalarParameterValueonMaterials (CustomCharacter|GetSkeletalMeshForPartOnPlayerPawn) "ReactivityStatus" 0.0)
    (Variables|Default|SetIsAthenaKillEmote false)))

(event Custom|HealingEmote
  (bind _returnvalue (CustomCharacter|GetSkeletalMeshForPartOnPlayerPawn))
  (if (not (or (or (Variables|Default|GetIsDamageReceivedEmote) (Variables|Default|GetIsAthenaKillEmote)) (Variables|Default|GetIsHealingEmote)))
    (Variables|Default|SetIsHealingEmote true)
    (Rendering|Material|SetScalarParameterValueonMaterials _returnvalue "HealingStartTime" (Utilities|Time|GetGameTimeinSeconds))
    (Rendering|Material|SetScalarParameterValueonMaterials _returnvalue "ReactivityStatus" 4.0)
    (Utilities|FlowControl|RetriggerableDelay 9.0)
    (Rendering|Material|SetScalarParameterValueonMaterials (CustomCharacter|GetSkeletalMeshForPartOnPlayerPawn) "ReactivityStatus" 0.0)
    (Variables|Default|SetIsHealingEmote false)))

(event Custom|OpenChestEmote
  (if (not (or (or (Variables|Default|GetIsDamageReceivedEmote) (Variables|Default|GetIsAthenaKillEmote)) (Variables|Default|GetIsHealingEmote)))
    (Rendering|Material|SetScalarParameterValueonMaterials (CustomCharacter|GetSkeletalMeshForPartOnPlayerPawn) "ReactivityStatus" 2.0)
    (Utilities|FlowControl|RetriggerableDelay (Variables|Default|GetEmoteDuration))
    (Rendering|Material|SetScalarParameterValueonMaterials (CustomCharacter|GetSkeletalMeshForPartOnPlayerPawn) "ReactivityStatus" 0.0)))

(fn ConstructionScript
  (|CallParentFunction))
