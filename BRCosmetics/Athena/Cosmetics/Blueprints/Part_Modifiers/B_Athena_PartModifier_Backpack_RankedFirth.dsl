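; Backpack part modifier for the RankedFirth cosmetic that adjusts the back bling's material in the front end.
; After the post res-in event runs the parent logic, it checks for the front end, waits briefly, then sets the DiamondEmissiveIntensity scalar on the materials of the part's skeletal mesh on the player pawn.
; The VFXUpdate override only forwards the progression level to its parent.
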(event EventOnPostResIn
  (|CallParentFunction)
  (if (CustomCosmeticModifier|IsinFrontEnd)
    (Utilities|FlowControl|Delay 0.1)
    (Rendering|Material|SetScalarParameterValueonMaterials (CustomCharacter|GetSkeletalMeshForPartOnPlayerPawn) "DiamondEmissiveIntensity" 2e+01)))

(fn ConstructionScript
  (|CallParentFunction))

(fn VFXUpdate
  (params (Progression%20Level :type Integer) (KeepIdleAlwaysOn :type Boolean))
  (|CallParentFunction Progression%20Level))
