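; Part modifier for the RaveNinja back bling that reacts to kills, shield gains and chest opens with sound and visuals.
; Cosmetic stat watchers stamp a trigger time into the mesh materials and play the matching sound, slightly delayed for chests.
; During playback, precomputed audio-analysis (NRT) loudness drives a material react curve and the idle VFX light multiplier, making visuals follow the sound.
; The trigger time is re-applied on post res-in.
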(event EventOnPawnAndPartChanged
  (|CallParentFunction)
  (Variables|Default|SetKillCount 0.0)
  (Utilities|FlowControl|Sequence
    (:then_0
      ;[1 1 1] On Kill Event
      (bind (_statname _statvalue) (Cosmetics|WatchCosmeticStat (CustomCharacter|GetAssociatedPlayerPawn) "(TagName=\"StatManager.AthenaKills\")" 0 false)
        (:OnStatChanged
          (bind _output_get (Variables|Default|SetKillCount (Math|Conversions|ToFloat(Integer) _statvalue)))
          (if (> _output_get 0.0)
            (CallFunction|TriggerMaterialFX)
            (CallFunction|SetandPlaySound 0)))))
    (:then_1
      ;[1 1 1] ShieldGain
      (bind (_statname_2 _statvalue_2) (Cosmetics|WatchCosmeticStat (CustomCharacter|GetAssociatedPlayerPawn) "(TagName=\"StatManager.GainedShieldTimes\")" 0 false)
        (:OnStatChanged
          (if (> _statvalue_2 0)
            (CallFunction|TriggerMaterialFX)
            (CallFunction|SetandPlaySound 2)))))
    (:then_2
      ;[1 1 1] Chest Open Event
      (bind (_statname_3 _statvalue_3) (Cosmetics|WatchCosmeticStat (CustomCharacter|GetAssociatedPlayerPawn) "(TagName=\"StatManager.OpenChests\")" 0 false)
        (:OnStatChanged
          (if (> _statvalue_3 0)
            (CallFunction|TriggerMaterialFX)
            (Utilities|FlowControl|Delay 0.35)
            (CallFunction|SetandPlaySound 1)))))))

(event OnAudioPlaybackPercent(Audio)
  (params (PlayingSoundWave :type Sound%20Wave%20Object%20Reference) (PlaybackPercent :type Float%20%28single-precision%29))
  (bind _audio (Variables|Default|GetAudio))
  (bind _sound (Variables|/Script/Engine.AudioComponent|GetSound _audio))
  (bind _duration (Variables|/Script/Engine.SoundBase|GetDuration _sound))
  (bind _returnvalue (Math|Float|MapRangeUnclamped PlaybackPercent 0.0 1.0 0.0 _duration))
  (bind _returnvalue_2 (CustomCharacter|GetSkeletalMeshForPartOnPlayerPawn))
  ;[1 1 1] Audio synesthesia parameter for visuals
  (bind _currentnrt (Variables|Default|GetCurrentNRT)
    (:then
      (Rendering|Material|SetScalarParameterValueonMaterials _returnvalue_2 "RM_KillReactDuration" _duration)
      (Rendering|Material|SetScalarParameterValueonMaterials _returnvalue_2 "RN_KillReactCurve" (AudioAnalyzer|GetNormalizedLoudnessatTime _currentnrt _returnvalue))
      (Effects|Components|ParticleSystem|SetFloatParameter (CallFunction|GetIdleVFXComponent) "User.LightMultiplier" (AudioAnalyzer|GetNormalizedLoudnessatTime _currentnrt _returnvalue)))
    (:else)))

(event EventOnPostResIn
  (|CallParentFunction)
  (CallFunction|UpdateMaterialParameters))

(fn SetAndPlaySound
  (params (Type :type Integer))
  (Audio|Components|Audio|SetSound (Variables|Default|GetAudio) (Utilities|Select :Index Type :Option%200 (Variables|Default|GetOnKillSound) :Option%201 (Variables|Default|GetOnChestOpenSound) :Option%202 (Variables|Default|GetOnShieldGainSound)))
  (Variables|Default|SetCurrentNRT (Utilities|Select :Index Type :Option%200 (Variables|Default|GetOnKillSoundNRT) :Option%201 (Variables|Default|GetOnChestOpenSoundNRT) :Option%202 (Variables|Default|GetOnShieldGainSoundNRT)))
  (Audio|Components|Audio|Play (Variables|Default|GetAudio) 0.0)
  (Transformation|AttachComponentToComponent (Variables|Default|GetAudio) (CustomCharacter|GetSkeletalMeshForPartOnPlayerPawn) "None" "SnapToTarget" "SnapToTarget" "SnapToTarget" true))

(fn TriggerMaterialFX
  (Variables|Default|SetTriggerTime (Utilities|Time|GetGameTimeinSeconds))
  (CallFunction|UpdateMaterialParameters))

(fn UpdateMaterialParameters
  (Rendering|Material|SetScalarParameterValueonMaterials (CustomCharacter|GetSkeletalMeshForPartOnPlayerPawn) "RN_GameTimeFromBP" (Variables|Default|GetTriggerTime)))

(fn ConstructionScript)
