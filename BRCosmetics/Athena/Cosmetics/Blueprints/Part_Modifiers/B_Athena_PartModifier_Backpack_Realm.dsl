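; Part modifier for the Realm back bling that plays a sound when the wearer is downed (DBNO).
; On BeginPlay it waits for playlist data, then binds to the owning pawn's DBNO-start event only if that pawn is locally controlled.
; The handler skips the front end, waits briefly, then spawns the cue attached to the part's skeletal mesh.
; The search-interact audio and stat-change events are stubs with no effect.
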(event EventBeginPlay
  (bind _returnvalue (CustomCharacter|GetAssociatedPlayerPawn))
  ;[1 1 1] Begin Play
  (Class|FortAsyncActionCurrentPlaylistReady|PlaylistDataReadyAsync
    (:OnReady
      (if (Pawn|IsLocallyControlled _returnvalue)
        (EventDispatchers|BindEventtoOnPlayerStartDBNO _returnvalue (delegate Custom|Trigger%20DBNO%20Audio))))))

(event Custom|Trigger%20DBNO%20Audio
  ;[1 1 1] DBNO Audio
  (if (not (CustomCosmeticModifier|IsinFrontEnd))
    (Utilities|FlowControl|Delay 0.5)
    (Audio|SpawnSoundAttached "/BRCosmetics/Sounds/Backblings/Realm/BackBling_Realm_DBNO_Cue.BackBling_Realm_DBNO_Cue" (CustomCharacter|GetSkeletalMeshForPartOnPlayerPawn))))

(event Custom|Started%20Interact%20Search%20Audio
  (if (CustomCosmeticModifier|IsinFrontEnd)
    _))

(event EventOnStatChanged
  (params (State%20Name :type Name :default "None") (Stat%20Value :type Integer)))

(fn ConstructionScript
  (|CallParentFunction))
