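; Backpack part modifier for the Renegade Raider cosmetic that sets up its orbit particle effect on the player pawn.
; After the parent post-res-in logic runs, it snaps the PSys-Orbit component to the Charm_Attach_RenegadeFire socket on the part's skeletal mesh.
; It then turns off custom depth rendering on the effect and copies the custom depth stencil value from the pawn's character mesh.
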(event EventOnPostResIn
  (bind _psys_orbit (Variables|Default|GetPSys-Orbit))
  (|CallParentFunction)
  (Transformation|AttachComponentToComponent (Variables|Default|GetPSys-Orbit) (CustomCharacter|GetSkeletalMeshForPartOnPlayerPawn) "Charm_Attach_RenegadeFire" "SnapToTarget" "SnapToTarget" "SnapToTarget" true)
  (Rendering|SetRenderCustomDepth _psys_orbit false)
  (Rendering|SetCustomDepthStencilValue _psys_orbit (Variables|/Script/Engine.PrimitiveComponent|GetCustomDepthStencilValue (Variables|/Script/Engine.Character|GetMesh (CustomCharacter|GetAssociatedPlayerPawn)))))

(fn ConstructionScript)
