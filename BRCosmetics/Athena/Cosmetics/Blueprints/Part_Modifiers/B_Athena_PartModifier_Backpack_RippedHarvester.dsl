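; Backpack cosmetic part modifier for the Ripped Harvester back bling that drives a material arrow from the wearer's movement direction.
; Each tick it transforms the player pawn's current movement acceleration into the pawn's local space.
; GetBasisDirection converts that vector's XY angle into a negated fraction of a full turn.
; The result is stored and written to the "arrow_direction" scalar parameter on the part's skeletal mesh materials.
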(event EventOnPostResIn
  (|CallParentFunction))

(event EventTick
  (params (DeltaSeconds :type Float%20%28single-precision%29))
  ;[1 1 1] Local Acceleration Direction
  (|CallParentFunction DeltaSeconds)
  (Utilities|FlowControl|Sequence
    (:then_0
      (bind _direction (CallFunction|GetBasisDirection (Math|Transform|InverseTransformDirection (Transformation|GetActorTransform (CustomCharacter|GetAssociatedPlayerPawn)) (Pawn|Components|CharacterMovement|GetCurrentAcceleration (Variables|/Script/Engine.Character|GetCharacterMovement (CustomCharacter|GetAssociatedPlayerPawn))))))
      (Variables|Default|Setdirection _direction))
    (:then_1
      ;[1 1 1] Set Material Params
      (Rendering|Material|SetScalarParameterValueonMaterials (CustomCharacter|GetSkeletalMeshForPartOnPlayerPawn) "arrow_direction" (Variables|Default|Getdirection)))))

(fn GetBasisDirection
  (params (Local%20Player%20Acceleration :type Vector))
  (return (* (/ (Math|Trig|Atan2(Degrees) (.y (Math|Vector|Normalize Local%20Player%20Acceleration 1.0)) (.x (Math|Vector|Normalize Local%20Player%20Acceleration 1.0))) 3.6e+02) -1.0)))

(fn ConstructionScript
  (|CallParentFunction))
