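; Cosmetic part modifier for the Robot backpack that shows the player's kill count on the part's skeletal mesh materials.
; It watches the AthenaKills stat once the pawn and part are set, and rereads it after res-in.
; Scalar parameters select a reticle at zero or split the count into tens and ones digits, leaving counts above 99 unapplied.
; The UpdateNumber timeline drives the FlashControl parameter to flash each update.
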(event Collision|EventActorBeginOverlap
  (params (OtherActor :type Actor%20Object%20Reference)))

(event EventTick
  (params (DeltaSeconds :type Float%20%28single-precision%29)))

(event EventOnPawnAndPartChanged
  (bind _returnvalue (CustomCharacter|GetAssociatedPlayerPawn))
  (bind _returnvalue_2 (CustomCharacter|GetCharacterPartType))
  (bind _returnvalue_3 (Customization|GetSkeletalMeshforPartType _returnvalue _returnvalue_2 :!class /Script/FortniteGame.FortPlayerPawn))
  (bind _killcount (Variables|Default|GetKillCount))
  (bind _returnvalue_4 (/ _killcount 1e+01))
  (bind _returnvalue_5 (Math|Float|Fraction _returnvalue_4))
  (bind _returnvalue_6 (* _returnvalue_5 1e+01))
  (|CallParentFunction)
  (bind (_statname _statvalue) (Cosmetics|WatchCosmeticStat _returnvalue "(TagName=\"StatManager.AthenaKills\")" 0 false)
    (:OnStatChanged
      (bind _output_get (Variables|Default|SetKillCount (Math|Conversions|ToFloat(Integer) _statvalue)))
      (if (> _output_get 0.0)
        (if (> _output_get 9.0)
          (if (not (> _output_get 99.0))
            (Rendering|Material|SetScalarParameterValueonMaterials _returnvalue_3 "UseReticle" 0.0)
            (Rendering|Material|SetScalarParameterValueonMaterials _returnvalue_3 "UseSingleDigit" 0.0)
            (Rendering|Material|SetScalarParameterValueonMaterials _returnvalue_3 "Kills-Tens" (Math|Conversions|ToFloat(Integer) (Math|Float|Floor _returnvalue_4)))
            (Rendering|Material|SetScalarParameterValueonMaterials _returnvalue_3 "Kills-Ones" _returnvalue_6)
            (Timeline|UpdateNumber PlayFromStart
              (:Update
                (Rendering|Material|SetScalarParameterValueonMaterials _returnvalue_3 "FlashControl" 1.0))
              (:Finished
                (Rendering|Material|SetScalarParameterValueonMaterials _returnvalue_3 "FlashControl" 0.0))))
          (else
            (Rendering|Material|SetScalarParameterValueonMaterials _returnvalue_3 "UseReticle" 0.0)
            ;[1 1 1] Use single digit
            (Rendering|Material|SetScalarParameterValueonMaterials _returnvalue_3 "UseSingleDigit" 1.0)
            (Rendering|Material|SetScalarParameterValueonMaterials _returnvalue_3 "Kills-Tens" 0.0)
            (Rendering|Material|SetScalarParameterValueonMaterials _returnvalue_3 "Kills-Ones" _returnvalue_6)
            (Timeline|UpdateNumber PlayFromStart)))
        (else
          (Rendering|Material|SetScalarParameterValueonMaterials _returnvalue_3 "UseReticle" 1.0)
          (Rendering|Material|SetScalarParameterValueonMaterials _returnvalue_3 "UseSingleDigit" 1.0)
          (Rendering|Material|SetScalarParameterValueonMaterials _returnvalue_3 "Kills-Tens" 0.0)
          (Rendering|Material|SetScalarParameterValueonMaterials _returnvalue_3 "Kills-Ones" 0.0)
          (Timeline|UpdateNumber PlayFromStart))))))

(event EventOnPostResIn
  (bind _returnvalue (CustomCharacter|GetAssociatedPlayerPawn))
  (bind _returnvalue_2 (CustomCharacter|GetCharacterPartType))
  (bind _returnvalue_3 (Customization|GetSkeletalMeshforPartType _returnvalue _returnvalue_2 :!class /Script/FortniteGame.FortPlayerPawn))
  (bind _killcount (Variables|Default|GetKillCount))
  (bind _returnvalue_4 (/ _killcount 1e+01))
  (bind _returnvalue_5 (Math|Float|Fraction _returnvalue_4))
  (bind _returnvalue_6 (* _returnvalue_5 1e+01))
  (|CallParentFunction)
  (bind _returnvalue_7 (Class|FortCosmeticFunctionLibrary|GetCosmeticStatValue (CustomCharacter|GetAssociatedPlayerPawn) "(TagName=\"StatManager.AthenaKills\")"))
  (Variables|Default|SetKillCount (Math|Conversions|ToFloat(Integer) _returnvalue_7))
  (if (> (stale EventOnPawnAndPartChanged _output_get) 0.0)
    (if (> (stale EventOnPawnAndPartChanged _output_get) 9.0)
      (if (not (> (stale EventOnPawnAndPartChanged _output_get) 99.0))
        (Rendering|Material|SetScalarParameterValueonMaterials _returnvalue_3 "UseReticle" 0.0)
        (Rendering|Material|SetScalarParameterValueonMaterials _returnvalue_3 "UseSingleDigit" 0.0)
        (Rendering|Material|SetScalarParameterValueonMaterials _returnvalue_3 "Kills-Tens" (Math|Conversions|ToFloat(Integer) (Math|Float|Floor _returnvalue_4)))
        (Rendering|Material|SetScalarParameterValueonMaterials _returnvalue_3 "Kills-Ones" _returnvalue_6)
        (Timeline|UpdateNumber PlayFromStart))
      (else
        (Rendering|Material|SetScalarParameterValueonMaterials _returnvalue_3 "UseReticle" 0.0)
        ;[1 1 1] Use single digit
        (Rendering|Material|SetScalarParameterValueonMaterials _returnvalue_3 "UseSingleDigit" 1.0)
        (Rendering|Material|SetScalarParameterValueonMaterials _returnvalue_3 "Kills-Tens" 0.0)
        (Rendering|Material|SetScalarParameterValueonMaterials _returnvalue_3 "Kills-Ones" _returnvalue_6)
        (Timeline|UpdateNumber PlayFromStart)))
    (else
      (Rendering|Material|SetScalarParameterValueonMaterials _returnvalue_3 "UseReticle" 1.0)
      (Rendering|Material|SetScalarParameterValueonMaterials _returnvalue_3 "UseSingleDigit" 1.0)
      (Rendering|Material|SetScalarParameterValueonMaterials _returnvalue_3 "Kills-Tens" 0.0)
      (Rendering|Material|SetScalarParameterValueonMaterials _returnvalue_3 "Kills-Ones" 0.0)
      (Timeline|UpdateNumber PlayFromStart))))

(fn ConstructionScript)
