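; Cosmetic part modifier for the Royal Dusk backpack that makes it react to the player's eliminations.
; A kill-stat watcher plays a timeline driving material parameters for scroll, glow and intensity, spawns an attached Niagara burst and calls PlayAudio.
; In the front end, the same reaction previews after a short post-res-in delay, with the Niagara system flagged as preview.
; PlayAudio skips the backpack sound in the front end.
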(event Custom|FrontEndDisplay
  (bind _returnvalue (CustomCharacter|GetSkeletalMeshForPartOnPlayerPawn))
  (Utilities|FlowControl|Sequence
    (:then_0
      (bind (_direction _linescroll _wpo _pinkscroll _pinkpatternscroll _lineintensity _coremultiply _linecurve _basecolormultiply _outerglow) (Timeline|EliminationTimeline PlayFromStart)
        (:Update
          (Rendering|Material|SetScalarParameterValueonMaterials _returnvalue "LineScroll" _linescroll)
          (Rendering|Material|SetScalarParameterValueonMaterials _returnvalue "WPO" _wpo)
          (Rendering|Material|SetScalarParameterValueonMaterials _returnvalue "PinkScroll" _pinkscroll)
          (Rendering|Material|SetScalarParameterValueonMaterials _returnvalue "PinkPatternScroll" _pinkpatternscroll)
          (Rendering|Material|SetScalarParameterValueonMaterials _returnvalue "LineIntensity" _lineintensity)
          (Rendering|Material|SetScalarParameterValueonMaterials _returnvalue "CoreMultiply" _coremultiply)
          (Rendering|Material|SetScalarParameterValueonMaterials _returnvalue "LineCurve" _linecurve)
          (Rendering|Material|SetScalarParameterValueonMaterials _returnvalue "BaseColorMultiply" _basecolormultiply)
          (Rendering|Material|SetScalarParameterValueonMaterials _returnvalue "OuterGlow" _outerglow))))
    (:then_1
      (bind _returnvalue_2 (PartModifier|SpawnSystem|SpawnSystemAttachedPartModifier "/Game/Accessories/FORT_Backpacks/Backpack_F_MED_RoyalDusk/FX/NS_MED_RoyalDusk_Pack_Reactive.NS_MED_RoyalDusk_Pack_Reactive" (CustomCharacter|GetSkeletalMeshForPartOnPlayerPawn) "pack" "0.000000,-1.500000,0.000000" "0, 0, 0" "SnapToTarget")
        (:Success
          (Niagara|SetNiagaraVariable(Bool) _returnvalue_2 "bIsFrontEndPreview" (CustomCosmeticModifier|IsinFrontEnd)))))
    (:then_2
      (CallFunction|PlayAudio))))

(event EventOnPawnAndPartChanged
  (bind _returnvalue (CustomCharacter|GetSkeletalMeshForPartOnPlayerPawn))
  (|CallParentFunction)
  ;[1 1 1] On Kill Event
  (Cosmetics|WatchCosmeticStat (CustomCharacter|GetAssociatedPlayerPawn) "(TagName=\"StatManager.AthenaKills\")" 0 false
    (:OnStatChanged
      (Utilities|FlowControl|Sequence
        (:then_0
          (Timeline|EliminationTimeline PlayFromStart))
        (:then_1
          (bind _returnvalue_2 (PartModifier|SpawnSystem|SpawnSystemAttachedPartModifier "/Game/Accessories/FORT_Backpacks/Backpack_F_MED_RoyalDusk/FX/NS_MED_RoyalDusk_Pack_Reactive.NS_MED_RoyalDusk_Pack_Reactive" (CustomCharacter|GetSkeletalMeshForPartOnPlayerPawn) "pack" "0.000000,-1.500000,0.000000" "0, 0, 0" "SnapToTarget")
            (:Success
              (Niagara|SetNiagaraVariable(Bool) _returnvalue_2 "bIsFrontEndPreview" (CustomCosmeticModifier|IsinFrontEnd)))))
        (:then_2
          (CallFunction|PlayAudio))))))

(event EventOnPostResIn
  ;[0 0 0] FrontEnd Display
  (|CallParentFunction)
  (if (CustomCosmeticModifier|IsinFrontEnd)
    (Utilities|FlowControl|Delay 3.6)
    (CallFunction|FrontEndDisplay)
    (Utilities|FlowControl|Delay 3.6)))

(fn PlayAudio
  (if (not (CustomCosmeticModifier|IsinFrontEnd))
    (bind _returnvalue (Audio|SpawnSoundAttached (Variables|Default|GetBackpackSound) (CustomCharacter|GetSkeletalMeshForPartOnPlayerPawn)))
    (Audio|Parameter|SetIntegerParameter _returnvalue "Stage" 0 :!class /Script/AudioExtensions.AudioParameterControllerInterface)))

(fn ConstructionScript)
