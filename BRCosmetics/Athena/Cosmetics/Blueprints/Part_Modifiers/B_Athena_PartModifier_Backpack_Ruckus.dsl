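; Part modifier for the Ruckus backpack cosmetic that tunes its material look for the front end versus other contexts.
; After the post-res-in event, it gets the part's skeletal mesh on the player pawn and calls the parent logic.
; An IsinFrontEnd branch then sets color grading, desaturation and emissive multiplier parameters on its materials.
; The front end gets grading, no desaturation and stronger emissive, while the color tint is identical in both branches.
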(event EventOnPostResIn
  (bind _returnvalue (CustomCharacter|GetSkeletalMeshForPartOnPlayerPawn))
  (|CallParentFunction)
  (if (CustomCosmeticModifier|IsinFrontEnd)
    (Rendering|Material|SetScalarParameterValueonMaterials _returnvalue "FTColorGrading" 0.4)
    (Rendering|Material|SetScalarParameterValueonMaterials _returnvalue "DesatuationAmount_FT" 0.0)
    (Rendering|Material|SetScalarParameterValueonMaterials _returnvalue "EmsMult_FT" 2.5)
    (Rendering|Material|SetVectorParameterValueonMaterials _returnvalue "ColorTint_FT" "0.947917,0.593788,0.527483")
    (else
      (Rendering|Material|SetScalarParameterValueonMaterials _returnvalue "FTColorGrading" 0.0)
      (Rendering|Material|SetScalarParameterValueonMaterials _returnvalue "DesatuationAmount_FT" 0.3)
      (Rendering|Material|SetScalarParameterValueonMaterials _returnvalue "EmsMult_FT" 2.0)
      (Rendering|Material|SetVectorParameterValueonMaterials _returnvalue "ColorTint_FT" "0.947917,0.593788,0.527483"))))

(fn ConstructionScript
  (|CallParentFunction))
