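; Part modifier for the Rumble backpack cosmetic that briefly swaps the backpack mesh for a reactive effect when a tracked stat changes to a positive value.
; An idle flag gates the swap, which kills VFX reactivity, activates the effect, hides the mesh through a material parameter and plays an attached sound.
; A retriggerable delay extends the reaction before restoring the mesh.
; In the frontend, a looping timer replays the sequence without sound.
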(event EventOnStatChanged
  (params (State%20Name :type Name :default "None") (Stat%20Value :type Integer))
  (bind _returnvalue (CustomCharacter|GetSkeletalMeshForPartOnPlayerPawn))
  (bind _returnvalue_2 (CallFunction|GetIdleVFXComponent))
  (if (> Stat%20Value 0)
    (if (Variables|Default|GetIs%20Idle)
      (Variables|Default|SetIs%20Idle false)
      (Effects|Components|ParticleSystem|SetBoolParameter _returnvalue_2 "bKillReactivity" true)
      (Utilities|FlowControl|Delay 1.0)
      (Components|Activation|Activate _returnvalue_2 true)
      (Rendering|Material|SetScalarParameterValueonMaterials _returnvalue "Hide Element 01" 1.0) ; Hides the backpack mesh, which is replaced by the effect
      (CallFunction|PlaySound _returnvalue)
      (else
        _))
    (Utilities|FlowControl|RetriggerableDelay 2.0)
    (else
      (Variables|Default|SetIs%20Idle true)))
  (Rendering|Material|SetScalarParameterValueonMaterials _returnvalue "Hide Element 01" 0.0) ; Brings back the backpack mesh
  (Components|Activation|Deactivate _returnvalue_2)
  (Effects|Components|ParticleSystem|SetBoolParameter _returnvalue_2 "bKillReactivity" false)
  (Variables|Default|SetIs%20Idle true))

(event Custom|Frontend_Event
  (bind _returnvalue (CustomCharacter|GetSkeletalMeshForPartOnPlayerPawn))
  (bind _returnvalue_2 (CallFunction|GetIdleVFXComponent))
  (Effects|Components|ParticleSystem|SetBoolParameter _returnvalue_2 "bKillReactivity" true)
  (Utilities|FlowControl|Delay 1.0)
  (Components|Activation|Activate _returnvalue_2 true)
  (Rendering|Material|SetScalarParameterValueonMaterials _returnvalue "Hide Element 01" 1.0) ; Hides the backpack mesh, which is replaced by the effect
  (CallFunction|PlaySound _returnvalue)
  (Utilities|FlowControl|RetriggerableDelay 2.0)
  (Rendering|Material|SetScalarParameterValueonMaterials _returnvalue "Hide Element 01" 0.0) ; Brings back the backpack mesh
  (Components|Activation|Deactivate _returnvalue_2)
  (Effects|Components|ParticleSystem|SetBoolParameter _returnvalue_2 "bKillReactivity" false)
  (Variables|Default|SetIs%20Idle true))

(event EventOnPostResIn
  ;[1 1 1] Front end reactivity cycle
  (|CallParentFunction)
  (if (Reactive|Frontend|ShouldDisplayFrontendReactivity 0 "Backpack")
    (Utilities|FlowControl|Delay 0.2)
    (Utilities|Time|SetTimerbyEvent (delegate Custom|Frontend_Event) 7.0 true)))

(fn PlaySound
  (params (AttachToComponent :type Scene%20Component%20Object%20Reference))
  (bind (_asfort_game_mode_front_end _bsuccess) (Utilities|Casting|CastToFortGameModeFrontEnd (Game|GetGameMode)))
  (if (not _bsuccess)
    (Audio|SpawnSoundAttached "/BRCosmetics/Sounds/Backblings/RumbleMale/BackBling_RumbleMale_React_Cue.BackBling_RumbleMale_React_Cue" AttachToComponent)))

(fn ConstructionScript
  (|CallParentFunction))
