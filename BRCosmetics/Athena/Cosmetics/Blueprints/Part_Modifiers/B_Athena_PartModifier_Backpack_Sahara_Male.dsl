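; Part modifier for the male Sahara backpack cosmetic that adds reactive idle VFX and a shake animation.
; After post-res-in, if frontend reactivity should display, it caches the backpack mesh's anim blueprint and plays a timeline.
; A timeline event activates the idle VFX and starts shaking, and completion stops the shake.
; A stat change above zero activates the same VFX and calls PlayAudio, whose body is empty here.
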(event EventOnPostResIn
  (bind _anim_bp (Variables|Default|GetAnim%20BP))
  (|CallParentFunction)
  (if (Reactive|Frontend|ShouldDisplayFrontendReactivity 0 "Backpack")
    (bind _asm_med_sahara_pack_anim_bp (Utilities|Casting|CastToM_MED_Sahara_Pack_AnimBP (Components|SkeletalMesh|GetAnimInstance (Customization|GetSkeletalMeshforPartType (CustomCharacter|GetAssociatedPlayerPawn) (CustomCharacter|GetCharacterPartType) :!class /Script/FortniteGame.FortPlayerPawn))))
    (Variables|Default|SetAnim%20BP _asm_med_sahara_pack_anim_bp)
    (Timeline|FrontEnd Play
      (:Finished
        (Variables|/BRCosmetics/Accessories/FORT_Backpacks/Backpack_M_MED_Sahara/Meshes/M_MED_Sahara_Pack_AnimBP.M_MED_Sahara_Pack_AnimBP_C|SetShakingAnimation false _anim_bp))
      (:Activate%20Front%20end
        (Components|Activation|Activate (CallFunction|GetIdleVFXComponent) false)
        (Variables|/BRCosmetics/Accessories/FORT_Backpacks/Backpack_M_MED_Sahara/Meshes/M_MED_Sahara_Pack_AnimBP.M_MED_Sahara_Pack_AnimBP_C|SetShakingAnimation true _anim_bp)))))

(event EventOnStatChanged
  (params (State%20Name :type Name :default "None") (Stat%20Value :type Integer))
  (|CallParentFunction State%20Name Stat%20Value)
  (if (> Stat%20Value 0)
    (Components|Activation|Activate (CallFunction|GetIdleVFXComponent) false)
    (CallFunction|PlayAudio)))

(fn PlayAudio)

(fn ConstructionScript
  (|CallParentFunction))
