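; Backpack cosmetic part modifier for the SailorSquad back bling that tunes its toon material specular when the character finishes resing in.
; After the parent resin logic, it runs the toon frontend LUT replace utility, then gets the part's skeletal mesh on the player pawn.
; It sets the SpecPower scalar on that mesh's materials, using a lower value in the frontend than in match.
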(event EventOnPostResIn
  (bind _returnvalue (CustomCharacter|GetSkeletalMeshForPartOnPlayerPawn))
  (|CallParentFunction)
  (Utilities|ToonFrontendLUTReplace
    (:Out
      (if (CustomCosmeticModifier|IsinFrontEnd)
        (Rendering|Material|SetScalarParameterValueonMaterials _returnvalue "SpecPower" 0.001)
        (else
          (Rendering|Material|SetScalarParameterValueonMaterials _returnvalue "SpecPower" 0.005))))))

(fn ConstructionScript
  (|CallParentFunction))
