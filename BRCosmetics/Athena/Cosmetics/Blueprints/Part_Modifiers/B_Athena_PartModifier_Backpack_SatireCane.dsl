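; Part modifier for the SatireCane back bling that reacts to the wearer's kills, or on a repeating timer in the frontend.
; It sets a Niagara variant index from the mesh name and keeps idle VFX off until a reaction.
; If the particle variant toggle allows, reactions play library sound, activate VFX, timestamp materials and pulse an AnimBP eliminated flag after a delay.
; A flag throttles in-game reactions during a short cooldown.
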(event EventOnPostResIn
  (|CallParentFunction)
  (Utilities|FlowControl|Sequence
    (:then_0
      (bind _asniagara_particle_system_component (Utilities|Casting|CastToNiagaraComponent (CallFunction|GetIdleVFXComponent)))
      (Niagara|SetNiagaraVariable(Int32) _asniagara_particle_system_component "VariantIndex" (CallFunction|FXGetVariantIndex)))
    (:then_1
      ; DEACTIVATE BY DEFAULT TO PREVENT FX SPAWNING WHEN TURNING ON/OFF THE REACTIVITY TOGGLE
      (Components|Activation|Deactivate (CallFunction|GetIdleVFXComponent))
      (if (Reactive|Frontend|ShouldDisplayFrontendReactivity 0 "Backpack")
        (Utilities|FlowControl|Delay 3.0)
        (CallFunction|OnFrontEndReactivity)
        (Utilities|Time|SetTimerbyEvent (EventDispatchers|CreateEvent :!event Custom|OnFrontEndReactivity) 9.0 true)
        (else
          (bind (_statname _statvalue) (Cosmetics|WatchCosmeticStat (CustomCharacter|GetAssociatedPlayerPawn) "(TagName=\"StatManager.AthenaKills\")" 0 false)
            (:OnStatChanged
              (if (not (Variables|Default|GetIsPlayingReactivity))
                (CallFunction|OnKillReact _statname _statvalue true)
                (Variables|Default|SetIsPlayingReactivity true)
                (Utilities|FlowControl|Delay 5.0)
                (Variables|Default|SetIsPlayingReactivity false)))))))))

(event Custom|OnFrontEndReactivity
  (CallFunction|OnKillReact "None"))

(event Custom|OnKillReact
  (params (StatName :type Name) (StatValue :type Integer) (AllowAudio :type Boolean))
  (bind _returnvalue (CallFunction|FXIsReactivityToggled))
  (if _returnvalue
    (Utilities|FlowControl|Sequence
      (:then_0
        (Class|BReactiveStatWatcherComponent|PlayLibrarySound (Variables|Default|GetB_Reactive_StatWatcher_Component) StatName StatValue (Variables|Default|GetLibrary%20Tag)))
      (:then_1
        (Components|Activation|Activate (CallFunction|GetIdleVFXComponent) true)
        (bind _asskeletal_mesh_component (Utilities|Casting|CastToSkeletalMeshComponent (CustomCharacter|GetSkeletalMeshForPartOnPlayerPawn))
          (:then
            (CallFunction|ToggleAnimBPData (Components|SkeletalMesh|GetAnimInstance _asskeletal_mesh_component)))
          (:CastFailed
            _))
        (for _array_element (Rendering|Material|GetMaterials (CustomCharacter|GetSkeletalMeshForPartOnPlayerPawn))
          (bind _asmaterial_instance_dynamic (Utilities|Casting|CastToMaterialInstanceDynamic _array_element))
          (Rendering|Material|SetScalarParameterValue _asmaterial_instance_dynamic "Reactivity_Timestamp" (Utilities|Time|GetGameTimeinSeconds) :!class /Script/Engine.MaterialInstanceDynamic)))
      (:then_2
        (Utilities|FlowControl|RetriggerableDelay 2.25)
        (bind _asskeletal_mesh_component_2 (Utilities|Casting|CastToSkeletalMeshComponent (CustomCharacter|GetSkeletalMeshForPartOnPlayerPawn)))
        (CallFunction|ToggleAnimBPData (Components|SkeletalMesh|GetAnimInstance _asskeletal_mesh_component_2) true)
        (Utilities|FlowControl|RetriggerableDelay 3.0)
        (CallFunction|ToggleAnimBPData (Components|SkeletalMesh|GetAnimInstance _asskeletal_mesh_component_2))))))

(fn FX_GetVariantIndex
  (bind _asskeletal_mesh_component (Utilities|Casting|CastToSkeletalMeshComponent (CustomCharacter|GetSkeletalMeshForPartOnPlayerPawn)))
  (bind _output_get (LocalVariables|Setobj_name (Utilities|GetObjectName (Variables|/Script/Engine.SkinnedMeshComponent|GetSkeletalMesh _asskeletal_mesh_component))))
  (bind _output_get_2 (LocalVariables|Setfound_index (Math|Integer|SelectInt 1 0 (Utilities|String|Contains _output_get "OdeYap"))))
  (return _output_get_2))

(fn FX_IsReactivityToggled
  (bind _returnvalue (CustomCharacter|GetAssociatedPlayerPawn))
  ; CHECK IF REACTIVITY TOGGLE IS ON/OFF
  (Utilities|IsValid _returnvalue
    (:Is%20Valid
      (bind _index (CosmeticLoadout|GetVariantChannelActiveIndexforSlotActor "(TagName=\"LoadoutSlot.Backpack\")" _returnvalue "(TagName=\"Cosmetics.Variant.Channel.Particle\")" (Utilities|Array|MakeArray "(TagName=\"Cosmetics.Variant.Property.Particle0\")" "(TagName=\"Cosmetics.Variant.Property.Particle1\")")))
      (return (== _index 0)))
    (:Is%20Not%20Valid
      (return false))))

(fn PlayLibrarySound
  (params (StatName :type Name :default "None") (StatValue :type Integer) (StatTag :type Fort%20Stat%20Manager%20Tag%20Structure%20%28by%20ref%29)))

(fn ToggleAnimBPData
  (params (AnimInstance :type Anim%20Instance%20Object%20Reference) (Eliminated :type Boolean))
  (Utilities|FlowControl|Sequence
    (:then_0
      (bind _asm_med_satire_cane_ode_pack_anim_bp (Utilities|Casting|CastToM_MED_SatireCane_Ode_Pack_AnimBP AnimInstance)
        (:then
          (Variables|/BRCosmetics/Accessories/FORT_Backpacks/Backpack_M_MED_SatireCane_Ode/Meshes/M_MED_SatireCane_Ode_Pack_AnimBP.M_MED_SatireCane_Ode_Pack_AnimBP_C|SetEnemyPlayerEliminated Eliminated _asm_med_satire_cane_ode_pack_anim_bp))
        (:CastFailed
          (bind _asm_med_satire_cane_ode_yap_pack_anim_bp (Utilities|Casting|CastToM_MED_SatireCane_OdeYap_Pack_AnimBP AnimInstance))
          (Variables|/BRCosmetics/Accessories/FORT_Backpacks/Backpack_M_MED_SatireCane_OdeYap/Meshes/M_MED_SatireCane_OdeYap_Pack_AnimBP.M_MED_SatireCane_OdeYap_Pack_AnimBP_C|SetEnemyPlayerEliminated Eliminated _asm_med_satire_cane_ode_yap_pack_anim_bp))))
    (:then_1
      (return))))

(fn TriggerLibrary)

(fn ConstructionScript
  (|CallParentFunction))
