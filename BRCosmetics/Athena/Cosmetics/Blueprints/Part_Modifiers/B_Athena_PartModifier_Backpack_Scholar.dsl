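; Part modifier for the Scholar backpack that plays a reactive glow and particle burst on kill stat changes.
; Particle variant picks Color_01 for the burst systems, while Color_02's select is fixed at index 0.
; Each reaction spawns halo then delayed burst Niagara effects, flags the idle VFX and runs an emissive timeline, clearing the flag afterwards.
; In the front end, the pack AnimBP's reactivity cue triggers the same sequence.
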(event EventOnPostResIn
  (|CallParentFunction)
  (bind _index (Class|FortCosmeticFunctionLibrary|GetVariantChannelActiveIndexForSlot "(TagName=\"LoadoutSlot.Backpack\")" (CosmeticLoadout|GetCosmeticLoadout) "(TagName=\"Cosmetics.Variant.Channel.Particle\")" (Utilities|Array|MakeArray "(TagName=\"Cosmetics.Variant.Property.Particle1\")" "(TagName=\"Cosmetics.Variant.Property.Particle2\")")))
  (Variables|Default|SetVariantColor_01 (Utilities|Select :Index _index :Option%200 "(R=0.650000,G=0.200000,B=0.650000,A=1.000000)" :Option%201 "(R=0.220504,G=0.651406,B=0.130281,A=1.000000)"))
  (Variables|Default|SetVariantColor_02 (Utilities|Select :Index 0 :Option%200 "(R=0.226966,G=0.095307,B=0.651406,A=1.000000)" :Option%201 "(R=0.043735,G=0.219526,B=0.054480,A=1.000000)"))
  (if (CustomCosmeticModifier|IsinFrontEnd)
    (bind _asf_med_scholar_pack_anim_bp (Utilities|Casting|CastToF_MED_Scholar_Pack_AnimBP (Components|SkeletalMesh|GetAnimInstance (Customization|GetSkeletalMeshforPartType (CustomCharacter|GetAssociatedPlayerPawn) (CustomCharacter|GetCharacterPartType) :!class /Script/FortniteGame.FortPlayerPawn))))
    (Default|BindEventtoReactivityCue _asf_med_scholar_pack_anim_bp (EventDispatchers|CreateEvent :!event Custom|FE_Reactive))))

(event Custom|FE_Reactive
  (bind _returnvalue (PartModifier|SpawnSystem|SpawnSystemAttachedPartModifier "/BRCosmetics/Accessories/FORT_Backpacks/Backpack_F_MED_Scholar/FX/NS_ScholarBackpack_BurstHalo.NS_ScholarBackpack_BurstHalo" (CustomCharacter|GetSkeletalMeshForPartOnPlayerPawn) "root")
    (:Success
      (Niagara|SetNiagaraVariable(LinearColor) _returnvalue "Color_01" (Variables|Default|GetVariantColor_01))
      (Niagara|SetNiagaraVariable(LinearColor) _returnvalue "Color_02" (Variables|Default|GetVariantColor_02))
      (Niagara|SetNiagaraVariable(Bool) _returnvalue "bIsFrontEndPreview" false))
    (:Failure
      _))
  (Utilities|FlowControl|Delay 0.3)
  (bind _returnvalue_2 (PartModifier|SpawnSystem|SpawnSystemAttachedPartModifier "/BRCosmetics/Accessories/FORT_Backpacks/Backpack_F_MED_Scholar/FX/NS_ScholarBackpack_Burst.NS_ScholarBackpack_Burst" (CustomCharacter|GetSkeletalMeshForPartOnPlayerPawn) "root")
    (:Success
      (Niagara|SetNiagaraVariable(LinearColor) _returnvalue_2 "Color_01" (Variables|Default|GetVariantColor_01))
      (Niagara|SetNiagaraVariable(LinearColor) _returnvalue_2 "Color_02" (Variables|Default|GetVariantColor_02))
      (Niagara|SetNiagaraVariable(Bool) _returnvalue_2 "bIsFrontEndPreview" false))
    (:Failure
      _))
  (Effects|Components|ParticleSystem|SetBoolParameter (CallFunction|GetIdleVFXComponent) "User.bisStatChange" true)
  (bind (_direction _newtrack_0) (Timeline|Timeline_0 PlayFromStart)
    (:Update
      (Rendering|Material|SetScalarParameterValueonMaterials (CustomCharacter|GetSkeletalMeshForPartOnPlayerPawn) "emissive mult" _newtrack_0))
    (:Finished
      (Utilities|FlowControl|Delay 2.0)
      (Effects|Components|ParticleSystem|SetBoolParameter (CallFunction|GetIdleVFXComponent) "User.bisStatChange" false))))

(event EventOnStatChanged
  (params (State%20Name :type Name :default "None") (Stat%20Value :type Integer))
  (|CallParentFunction State%20Name Stat%20Value))

(event EventOnPawnAndPartChanged
  (|CallParentFunction)
  (Cosmetics|WatchCosmeticStat (CustomCharacter|GetAssociatedPlayerPawn) "(TagName=\"StatManager.AthenaKills\")" 0 false
    (:OnStatChanged
      (bind _returnvalue (PartModifier|SpawnSystem|SpawnSystemAttachedPartModifier "/BRCosmetics/Accessories/FORT_Backpacks/Backpack_F_MED_Scholar/FX/NS_ScholarBackpack_BurstHalo.NS_ScholarBackpack_BurstHalo" (CustomCharacter|GetSkeletalMeshForPartOnPlayerPawn) "root")
        (:Success
          (Niagara|SetNiagaraVariable(LinearColor) _returnvalue "Color_01" (Variables|Default|GetVariantColor_01))
          (Niagara|SetNiagaraVariable(LinearColor) _returnvalue "Color_02" (Variables|Default|GetVariantColor_02))
          (Niagara|SetNiagaraVariable(Bool) _returnvalue "bIsFrontEndPreview" false))
        (:Failure
          _))
      (Utilities|FlowControl|Delay 0.3)
      (bind _returnvalue_2 (PartModifier|SpawnSystem|SpawnSystemAttachedPartModifier "/BRCosmetics/Accessories/FORT_Backpacks/Backpack_F_MED_Scholar/FX/NS_ScholarBackpack_Burst.NS_ScholarBackpack_Burst" (CustomCharacter|GetSkeletalMeshForPartOnPlayerPawn) "root")
        (:Success
          (Niagara|SetNiagaraVariable(LinearColor) _returnvalue_2 "Color_01" (Variables|Default|GetVariantColor_01))
          (Niagara|SetNiagaraVariable(LinearColor) _returnvalue_2 "Color_02" (Variables|Default|GetVariantColor_02))
          (Niagara|SetNiagaraVariable(Bool) _returnvalue_2 "bIsFrontEndPreview" false))
        (:Failure
          _))
      (Effects|Components|ParticleSystem|SetBoolParameter (CallFunction|GetIdleVFXComponent) "User.bisStatChange" true)
      (Timeline|Timeline_0 PlayFromStart))))

(fn ConstructionScript
  (|CallParentFunction))
