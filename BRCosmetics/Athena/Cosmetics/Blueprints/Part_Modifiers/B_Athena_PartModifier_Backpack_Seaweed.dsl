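; Backpack part modifier for the Seaweed back bling cosmetic that plays a burst effect when the pawn jumps.
; On the jump event, after calling the parent handler, it spawns the Seaweed burst Niagara system attached to the "pack" point on the associated pawn's skeletal mesh for its part type.
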(event EventOnPawnJumped
  (|CallParentFunction)
  (PartModifier|SpawnSystem|SpawnSystemAttachedPartModifier "/BRCosmetics/Accessories/FORT_Backpacks/Backpack_M_MED_Seaweed/FX/NS_Backpack_Seaweed_Burst.NS_Backpack_Seaweed_Burst" (Customization|GetSkeletalMeshforPartType (CustomCharacter|GetAssociatedPlayerPawn) (CustomCharacter|GetCharacterPartType) :!class /Script/FortniteGame.FortPlayerPawn) "pack"))

(fn ConstructionScript)
