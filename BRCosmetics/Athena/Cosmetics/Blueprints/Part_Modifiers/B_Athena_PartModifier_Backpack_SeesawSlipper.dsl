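; Backpack part modifier Blueprint for the SeesawSlipper cosmetic that tunes its idle VFX for the current translucency setting.
; On the post-res-in event it calls the parent, reads the r.SeparateTranslucency console variable and writes it to the idle particle component's SepTrans float parameter.
; The construction script only calls the parent.
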(event EventOnPostResIn
  (|CallParentFunction)
  (bind _returnvalue (Development|GetConsoleVariableFloatValue "r.SeparateTranslucency"))
  (Effects|Components|ParticleSystem|SetFloatParameter (CallFunction|GetIdleVFXComponent) "SepTrans" _returnvalue))

(fn ConstructionScript
  (|CallParentFunction))
