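; Cosmetic part modifier Blueprint for the Shiitake Shaolin back bling that plays a landing effect.
; When the pawn lands, it calls the parent handler, then spawns the backpack's Jump Niagara system attached to the part's skeletal mesh on the player pawn at FX_Base.
; The construction script is empty.
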(event EventOnPawnLanded
  (|CallParentFunction)
  (PartModifier|SpawnSystem|SpawnSystemAttachedPartModifier "/BRCosmetics/Accessories/FORT_Backpacks/M_MED_Shiitake_Shaolin_Backpack/FX/NS_Backpack_ShiitakeShaolin_Jump.NS_Backpack_ShiitakeShaolin_Jump" (CustomCharacter|GetSkeletalMeshForPartOnPlayerPawn) "FX_Base"))

(fn ConstructionScript)
