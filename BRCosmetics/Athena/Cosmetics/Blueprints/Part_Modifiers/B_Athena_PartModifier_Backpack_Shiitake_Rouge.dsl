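; Backpack cosmetic part modifier for the Shiitake Shaolin Rogue skin that plays a landing effect on the player pawn.
; On the pawn-landed event it calls the parent behavior, then spawns the backpack's Niagara jump system attached to the part's skeletal mesh at the FX_Base socket.
; The construction script is empty.
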(event EventOnPawnLanded
  (|CallParentFunction)
  (PartModifier|SpawnSystem|SpawnSystemAttachedPartModifier "/BRCosmetics/Accessories/FORT_Backpacks/M_MED_Shiitake_Shaolin_Backpack/Skins/Rogue/FX/NS_Backpack_ShiitakeShaolin_Rouge_Jump.NS_Backpack_ShiitakeShaolin_Rouge_Jump" (CustomCharacter|GetSkeletalMeshForPartOnPlayerPawn) "FX_Base"))

(fn ConstructionScript)
