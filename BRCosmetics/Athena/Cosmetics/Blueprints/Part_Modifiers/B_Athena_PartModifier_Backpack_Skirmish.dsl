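; Back bling part modifier for the Skirmish backpack that swaps its screen material in reaction to the wearer's game state.
; After a delayed setup, kill stat, health, safe zone, DBNO and game phase events feed a priority-based screen state machine.
; Each change preserves the chrome parameter, matches the attached glow material to the screen and plays kill or safe zone audio.
; Kill screens and their paired sounds are random and revert on a timer.
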(event EventOnPostResIn
  ;[1 1 1] Begin Play
  (|CallParentFunction)
  (Utilities|FlowControl|Delay 3.0)
  (Transformation|AttachComponentToComponent (Variables|Default|GetScreenGlow) (CustomCharacter|GetSkeletalMeshForPartOnPlayerPawn) "ScreenLightGlow" "KeepRelative" "KeepRelative" "KeepRelative" true)
  (bind _output_get (Variables|Default|SetAssociatedPlayerPawn (CustomCharacter|GetAssociatedPlayerPawn)))
  (EventDispatchers|AssignOnSafeZoneOccupancyChangedEvent _output_get (EventDispatchers|CreateEvent :!event Custom|OnSafeZoneOccupancyChangedEvent))
  (bind (_statname _statvalue) (Cosmetics|WatchCosmeticStat (Variables|Default|GetAssociatedPlayerPawn) "(TagName=\"StatManager.AthenaKills\")" -1 false)
    (:then
      (Health|AssignOnPawnHealthChanged (Variables|Default|GetAssociatedPlayerPawn) (EventDispatchers|CreateEvent :!event Custom|OnPawnHealthChanged_Event_Local))
      (CallFunction|GameStateChanged "None")
      (bind _gamephaseupdatedasynctask (Variables|Default|GetGamePhaseUpdatedAsyncTask)
        (:then
          (GameplayEventRouter|StopListeningtoStatefulEvent _gamephaseupdatedasynctask))
        (:else
          _))
      (bind _returnvalue (GameplayEventRouter|GetEventRouter "/Script/FortniteGame.FortGameState" self))
      (bind (_asynctaskproxy _payload) (GameplayEventRouter|StartListeningToStatefulEvent _returnvalue "/Script/FortniteGame.FortGamePhaseUpdatedEvent")
        (:then
          (Variables|Default|SetGamePhaseUpdatedAsyncTask _asynctaskproxy)
          (EventDispatchers|BindEventtoOnDBNOStateChanged (Variables|Default|GetAssociatedPlayerPawn) (EventDispatchers|CreateEvent :!event Custom|DBNOStateChanged)))
        (:OnEventReceived
          _)
        (:OnSavedEventStateExists
          _))
      (CallFunction|GameStateChanged (Utilities|Struct|BreakFortGamePhaseUpdatedEvent _payload)))
    (:OnStatChanged
      ;[1 1 1] Player Kill
      (if (> _statvalue 0)
        (CallFunction|UpdateScreenState "NewEnumerator2")))))

(event EventEndPlay
  (params (EndPlayReason :type EEndPlayReason%20Enum :default "Destroyed"))
  ;[1 1 1] On unequip - Is this needed?
  (|CallParentFunction EndPlayReason)
  (bind _associatedplayerpawn (Variables|Default|GetAssociatedPlayerPawn)
    (:then
      (EventDispatchers|UnbindEventfromOnSafeZoneOccupancyChangedEvent _associatedplayerpawn (EventDispatchers|CreateEvent :!event Custom|OnSafeZoneOccupancyChangedEvent))
      (Health|UnbindEventfromOnPawnHealthChanged _associatedplayerpawn (EventDispatchers|CreateEvent :!event Custom|OnPawnHealthChanged_Event_Local))
      (bind _gamephaseupdatedasynctask (Variables|Default|GetGamePhaseUpdatedAsyncTask)
        (:then
          (GameplayEventRouter|StopListeningtoStatefulEvent _gamephaseupdatedasynctask))
        (:else
          _))
      (bind _backbling_oneshot_audio_component (Variables|Default|GetBackBling%20Oneshot%20Audio%20Component)
        (:then
          (Audio|Components|Audio|FadeOut _backbling_oneshot_audio_component 0.2 0.0 "Linear" :!class /Script/Engine.AudioComponent))
        (:else)))
    (:else)))

(event Custom|GameStateChanged
  (params (GamePhase :type EAthenaGamePhase%20Enum))
  ;[1 1 1] Game State Changed
  (switch Utilities|FlowControl|Switch|SwitchonEAthenaGamePhase GamePhase
    (:None
      _)
    (:Setup
      (CallFunction|UpdateScreenState "NewEnumerator3"))
    (:Warmup
      _)
    (:Aircraft
      _)
    (:SafeZones
      _)
    (:EndGame
      _)
    (:Count
      _))
  (CallFunction|UpdateScreenState "NewEnumerator0" (Utilities|Enum|Equal(Enum) (Variables|Default|GetCurrentScreenState) "NewEnumerator3")))

(event Custom|DBNOStateChanged
  (params (FortPawn :type Fort%20Pawn%20Object%20Reference) (bInIsDBNO :type Boolean))
  ;[1 1 1] DBNO
  (if bInIsDBNO
    (CallFunction|UpdateScreenState "NewEnumerator5")
    (else
      (CallFunction|UpdateScreenState "NewEnumerator0" true))))

(event Custom|UpdateScreenState
  (params (ScreenState :type Enum_Skirmish_ScreenStates%20Enum :default "NewEnumerator0") (IgnoreSortPriority :type Boolean))
  (bind _returnvalue (CustomCharacter|GetSkeletalMeshForPartOnPlayerPawn))
  ;[1 1 1] Screen State Machine
  (if (or IgnoreSortPriority (> (CallFunction|GetScreenPriority ScreenState) (CallFunction|GetScreenPriority (Variables|Default|GetCurrentScreenState))))
    (Variables|Default|SetPreviousScreenState (Variables|Default|GetCurrentScreenState))
    (bind _output_get (Variables|Default|SetCurrentScreenState ScreenState))
    (Utilities|Time|ClearandInvalidateTimerbyHandle (Variables|Default|GetRevertScreenStateTimer))
    ;;[1 1 1] logic added to preserve chrome parameters as materials are swapped
    (bind _returnvalue_2 (Rendering|Material|GetScalarParameterValue (Material|GetMID _returnvalue) "S22ChromeCharacters_strength" :!class /Script/Engine.MaterialInstanceDynamic))
    (Rendering|Material|SetMaterial _returnvalue 1 (CallFunction|GetScreenMaterial _output_get))
    ;;[1 1 1] chrome
    (Rendering|Material|SetScalarParameterValueonMaterials _returnvalue "S22ChromeCharacters_strength" _returnvalue_2)
    (Rendering|Material|SetMaterial (Variables|Default|GetScreenGlow) 0 (CallFunction|GetFakeLightningMaterial))
    (CallFunction|PlaySound)
    (bind _returnvalue_3 (Utilities|Time|SetTimerbyEvent (delegate Custom|RevertScreenState) (CallFunction|GetScreenRevertDelay)))
    (Variables|Default|SetRevertScreenStateTimer _returnvalue_3)))

(event Custom|RevertScreenState
  ;[1 1 1] Screen State Machine
  (CallFunction|UpdateScreenState (Variables|Default|GetPreviousScreenState) true))

(event Custom|OnSafeZoneOccupancyChangedEvent
  (params (PlayerPawn :type Fort%20Player%20Pawn%20Object%20Reference) (bIsInSafeZone :type Boolean))
  ;[1 1 1] Safe Zone Changed
  (if bIsInSafeZone
    (CallFunction|ResetPreviousScreenState)
    (if (Utilities|Enum|Equal(Enum) (Variables|Default|GetCurrentScreenState) "NewEnumerator4")
      (CallFunction|UpdateScreenState (select (Variables|Default|GetLowHealth) "NewEnumerator1" "NewEnumerator0") true))
    (else
      (Variables|Default|SetPreviousScreenState "NewEnumerator4")
      (CallFunction|UpdateScreenState "NewEnumerator4"))))

(event Custom|OnPawnHealthChanged_Event_Local
  ;[1 1 1] Low Health
  (bind _associatedplayerpawn (Variables|Default|GetAssociatedPlayerPawn)
    (:then
      (if (<= (FortPawn|GetHealth _associatedplayerpawn) (Variables|Default|GetLowHealthThreshold))
        (Variables|Default|SetPreviousScreenState "NewEnumerator1")
        (if (not (Utilities|Enum|Equal(Enum) (Variables|Default|GetCurrentScreenState) "NewEnumerator3"))
          (CallFunction|UpdateScreenState "NewEnumerator1")
          (Variables|Default|SetLowHealth true))
        (else
          (Variables|Default|SetLowHealth false)
          (if (Utilities|Enum|Equal(Enum) (Variables|Default|GetCurrentScreenState) "NewEnumerator1")
            (CallFunction|UpdateScreenState "NewEnumerator0" true)))))
    (:else)))

(event Custom|DebugScreenState
  (params (Index :type Integer))
  ;[1 1 1] Debug Screen State
  (CallFunction|UpdateScreenState (Utilities|Select :Index Index :Option%200 "NewEnumerator0" :Option%201 "NewEnumerator1" :Option%202 "NewEnumerator2" :Option%203 "NewEnumerator3" :Option%204 "NewEnumerator4") true))

(fn GetFakeLightningMaterial
  (bind _returnvalue (CustomCharacter|GetSkeletalMeshForPartOnPlayerPawn))
  (bind _returnvalue_2 (Rendering|Material|GetMaterial _returnvalue 1))
  (bind _returnvalue_3 (Utilities|GetObjectName _returnvalue_2))
  (bind _fakelightingmaterials (Variables|Default|GetFakeLightingMaterials))
  (bind (_value _returnvalue_4) (Utilities|Map|Find _fakelightingmaterials _returnvalue_3))
  (if _returnvalue_4
    (return _value)
    (else
      (return "/BRCosmetics/Accessories/FORT_Backpacks/Backpack_F_MED_Skirmish/Materials/M_Skirmish_FakeLight_06_Inst.M_Skirmish_FakeLight_06_Inst"))))

(fn GetScreenMaterial
  (params (Index :type Enum_Skirmish_ScreenStates%20Enum :default "NewEnumerator0"))
  (bind _killscreenmaterials (Variables|Default|GetKillScreenMaterials))
  (bind _output_get (Variables|Default|SetKillScreenIndex (Math|Random|RandomInteger (Utilities|Array|Length _killscreenmaterials))))
  (return (Utilities|Select :Index Index :NewEnumerator0 "/Game/Accessories/FORT_Backpacks/Backpack_F_MED_Skirmish/Materials/M_Skirmish_Screen_06_Inst.M_Skirmish_Screen_06_Inst" :NewEnumerator1 "/BRCosmetics/Accessories/FORT_Backpacks/Backpack_F_MED_Skirmish/Materials/M_Skirmish_Screen_10_Inst.M_Skirmish_Screen_10_Inst" :NewEnumerator2 (Utilities|Array|Get(acopy) _killscreenmaterials _output_get) :NewEnumerator3 "/BRCosmetics/Accessories/FORT_Backpacks/Backpack_F_MED_Skirmish/Materials/M_Skirmish_Screen_05_Inst.M_Skirmish_Screen_05_Inst" :NewEnumerator4 "/BRCosmetics/Accessories/FORT_Backpacks/Backpack_F_MED_Skirmish/Materials/M_Skirmish_Screen_07_Inst.M_Skirmish_Screen_07_Inst" :NewEnumerator5 "/BRCosmetics/Accessories/FORT_Backpacks/Backpack_F_MED_Skirmish/Materials/M_Skirmish_Screen_10_Inst.M_Skirmish_Screen_10_Inst")))

(fn GetScreenPriority
  (params (ScreenState :type Enum_Skirmish_ScreenStates%20Enum :default "NewEnumerator0"))
  (return (Utilities|Select :Index ScreenState :NewEnumerator0 0 :NewEnumerator1 2 :NewEnumerator2 4 :NewEnumerator3 1 :NewEnumerator4 3 :NewEnumerator5 5)))

(fn GetScreenRevertDelay
  (return (Utilities|Select :Index (Variables|Default|GetCurrentScreenState) :NewEnumerator0 0.0 :NewEnumerator1 0.0 :NewEnumerator2 4.0 :NewEnumerator3 0.0 :NewEnumerator4 0.0 :NewEnumerator5 0.0)))

(fn PlaySound
  (bind _returnvalue (Audio|SpawnSoundAttached (Utilities|Select :Index (Variables|Default|GetCurrentScreenState) :NewEnumerator2 (Utilities|Map|Find (Variables|Default|GetKillScreenSoundMap) (Variables|Default|GetKillScreenIndex)) :NewEnumerator4 (Variables|Default|GetSafeZone%20Sound)) (CustomCharacter|GetSkeletalMeshForPartOnPlayerPawn)))
  (Variables|Default|SetBackBling%20Oneshot%20Audio%20Component _returnvalue))

(fn ResetPreviousScreenState
  (Variables|Default|SetPreviousScreenState (select (Variables|Default|GetLowHealth) "NewEnumerator1" "NewEnumerator0")))

(fn ConstructionScript
  (|CallParentFunction))
