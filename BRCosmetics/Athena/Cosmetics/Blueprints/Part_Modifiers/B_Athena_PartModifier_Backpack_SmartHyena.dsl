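; Part modifier Blueprint for the SmartHyena back bling, set up to react when its player takes damage.
; After res-in, it watches the associated pawn's DamageReceived cosmetic stat and calls TookDamage on each change.
; It then sets ReactivityEnabled according to whether the loadout's Parts variant channel resolves to Stage1.
; TookDamage checks that flag, but its enabled branch is empty, so no visible reaction happens here.
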(event Custom|TookDamage
  (if (Variables|Default|GetReactivityEnabled)
    _))

(event EventOnPostResIn
  (|CallParentFunction)
  (Utilities|FlowControl|Sequence
    (:then_0
      (Cosmetics|WatchCosmeticStat (CustomCharacter|GetAssociatedPlayerPawn) "(TagName=\"StatManager.DamageReceived\")" 0 false
        (:OnStatChanged
          (CallFunction|TookDamage))))
    (:then_1
      ;[1 1 1] Read CIDs
      (bind (_index _bindexisvalid) (Class|FortCosmeticFunctionLibrary|GetVariantChannelActiveIndexForItem "/BRCosmetics/Athena/Items/Cosmetics/Backpacks/Backpack_SmartHyena.Backpack_SmartHyena" (CosmeticLoadout|GetCosmeticLoadout) "(TagName=\"Cosmetics.Variant.Channel.Parts\")" (Utilities|Array|MakeArray "(TagName=\"Cosmetics.Variant.Property.Stage1\")") false false))
      (Variables|Default|SetReactivityEnabled _bindexisvalid))))

(fn ConstructionScript
  (|CallParentFunction))
