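; Part modifier Blueprint for the SmoothBeanie back bling that feeds pawn movement state to its idle VFX.
; On pawn landing, it writes the current game time to the LandGameTime float parameter on the idle VFX component.
; Every tick, it sets the bPawnSprintOrSlide bool parameter to true while the associated player pawn is tactical sprinting or sliding.
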(event EventOnPostResIn
  (|CallParentFunction))

(event EventOnPawnLanded
  (|CallParentFunction)
  (Effects|Components|ParticleSystem|SetFloatParameter (CallFunction|GetIdleVFXComponent) "LandGameTime" (Utilities|Time|GetGameTimeinSeconds)))

(event EventTick
  (params (DeltaSeconds :type Float%20%28single-precision%29))
  (Effects|Components|ParticleSystem|SetBoolParameter (CallFunction|GetIdleVFXComponent) "bPawnSprintOrSlide" (or (Variables|/Script/FortniteGame.FortPlayerPawn|GetbIsTacticalSprinting (CustomCharacter|GetAssociatedPlayerPawn)) (Variables|/Script/FortniteGame.FortPawn|GetbIsSliding (CustomCharacter|GetAssociatedPlayerPawn)))))

(fn ConstructionScript
  (|CallParentFunction))
