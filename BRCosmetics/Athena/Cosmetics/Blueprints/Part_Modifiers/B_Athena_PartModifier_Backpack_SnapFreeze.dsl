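; Cosmetic part modifier for the SnapFreeze back bling that triggers a timed material effect on the part's skeletal mesh when the player gets eliminations.
; After res-in outside the front end, it watches the AthenaKills stat and stamps the current game time into the Elim_StartTime material parameter on each change.
; In the front end, a self-calling delayed event replays the effect on a repeating timer after an initial delay.
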(event Custom|FrontendLoop
  ; Frontend Loop
  (Rendering|Material|SetScalarParameterValueonMaterials (CustomCharacter|GetSkeletalMeshForPartOnPlayerPawn) "Elim_StartTime" (Utilities|Time|GetGameTimeinSeconds))
  (Utilities|FlowControl|Delay 7.5)
  (CallFunction|FrontendLoop))

(event EventOnPostResIn
  (|CallParentFunction)
  (if (CustomCosmeticModifier|IsinFrontEnd)
    ; Frontend Loop
    (Utilities|FlowControl|Delay 5.0)
    (Rendering|Material|SetScalarParameterValueonMaterials (CustomCharacter|GetSkeletalMeshForPartOnPlayerPawn) "Elim_StartTime" (Utilities|Time|GetGameTimeinSeconds))
    (Utilities|FlowControl|Delay 7.5)
    (CallFunction|FrontendLoop)
    (else
      ; Material effect on Elimination
      (Cosmetics|WatchCosmeticStat (CustomCharacter|GetAssociatedPlayerPawn) "(TagName=\"StatManager.AthenaKills\")" 0 false
        (:OnStatChanged
          (Rendering|Material|SetScalarParameterValueonMaterials (CustomCharacter|GetSkeletalMeshForPartOnPlayerPawn) "Elim_StartTime" (Utilities|Time|GetGameTimeinSeconds)))))))

(fn ConstructionScript
  (|CallParentFunction))
