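; Cosmetic part modifier for the SoupGoal back bling that makes its material and idle VFX react to player stat changes.
; Each trigger increments a counter, passes it to the idle VFX and stamps game time into one of five rotating ReactivityStartTime material parameters, presumably so hits can overlap.
; In the front end, it sets a single-hit strength after res-in, then triggers reactivity after a short delay and on a looping timer.
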(event EventOnStatChanged
  (params (State%20Name :type Name :default "None") (Stat%20Value :type Integer))
  (|CallParentFunction State%20Name Stat%20Value)
  (CallFunction|TriggerReactivity))

(event EventOnPostResIn
  (|CallParentFunction)
  (if (CustomCosmeticModifier|IsinFrontEnd)
    (Rendering|Material|SetScalarParameterValueonMaterials (CustomCharacter|GetSkeletalMeshForPartOnPlayerPawn) "ReactivitySingleHitStrength" 0.8)
    (Utilities|FlowControl|Delay 3.0)
    (CallFunction|TriggerReactivity)
    (Utilities|Time|SetTimerbyEvent (delegate Custom|TriggerReactivity) 8.0 true)))

(event Custom|TriggerReactivity
  (Math|Integer|IncrementInt (Variables|Default|GetReactivityCounter)
    (:%20%20
      (Effects|Components|ParticleSystem|SetIntParameter (CallFunction|GetIdleVFXComponent) "ReactivityCounter" (Variables|Default|GetReactivityCounter))
      (Rendering|Material|SetScalarParameterValueonMaterials (CustomCharacter|GetSkeletalMeshForPartOnPlayerPawn) (Utilities|String|StringToName (Utilities|String|Append "ReactivityStartTime_" (Utilities|String|ToString(Integer) (Math|Integer|%25%28Integer%29 (Variables|Default|GetReactivityCounter) 5)))) (Utilities|Time|GetGameTimeinSeconds)))))

(fn ConstructionScript
  (|CallParentFunction))
