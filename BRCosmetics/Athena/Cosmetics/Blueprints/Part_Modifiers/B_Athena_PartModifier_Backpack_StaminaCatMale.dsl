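; Backpack part modifier for the male StaminaCat back bling, adding landing and idle effect handling.
; On pawn landing, a hidden-state check gates spawning a Niagara jump effect at the FX_BowlCenter socket of the part's skeletal mesh.
; After res-in, it takes the idle VFX component and, if valid, calls a toon frontend LUT replace.
; Both effects get a lowered translucent sort priority, which keeps the idle effect rendering behind the glass.
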(event EventOnPawnLanded
  (|CallParentFunction)
  (if (CallFunction|IsPartModifierHidden)
    (bind _returnvalue (PartModifier|SpawnSystem|SpawnSystemAttachedPartModifier "/BRCosmetics/Accessories/FORT_Backpacks/Backpack_M_MED_StaminaCat/FX/NS_Backpack_StaminaCatMale_Jump.NS_Backpack_StaminaCatMale_Jump" (CustomCharacter|GetSkeletalMeshForPartOnPlayerPawn) "FX_BowlCenter")
      (:Success
        (Rendering|SetTranslucentSortPriority _returnvalue -1)))))

(event EventOnPostResIn
  (bind _returnvalue (CallFunction|GetIdleVFXComponent))
  (|CallParentFunction)
  (Utilities|IsValid _returnvalue
    (:Is%20Valid
      ;[0.07292 0.07292 0.07292 0.224] Render behind the glass
      (Rendering|SetTranslucentSortPriority _returnvalue -1)
      (Utilities|ToonFrontendLUTReplace))))

(fn ConstructionScript)
