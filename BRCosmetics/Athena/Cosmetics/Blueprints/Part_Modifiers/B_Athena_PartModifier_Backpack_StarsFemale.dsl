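; Backpack cosmetic part modifier Blueprint for the StarsFemale back bling that toggles its idle trail effect around skydiving.
; When skydiving begins, it calls the parent handler and sets the idle VFX component's TrailOn float parameter to zero, turning the trail off.
; When skydiving ends, it calls the parent and restores TrailOn to a positive value.
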(event EventOnBeginSkydiving
  (|CallParentFunction)
  (Effects|Components|ParticleSystem|SetFloatParameter (CallFunction|GetIdleVFXComponent) "TrailOn" 0.0))

(event EventOnEndSkydiving
  (|CallParentFunction)
  (Effects|Components|ParticleSystem|SetFloatParameter (CallFunction|GetIdleVFXComponent) "TrailOn" 15.0))

(fn ConstructionScript)
