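; Part modifier Blueprint for the Summer Hotdog back bling that makes its effects react to the pawn jumping and landing.
; On jump it sets the idle VFX component's IsOnGround bool parameter to false, and on landing it sets it back to true.
; Landing also spawns the NS_HotDog_Summer_Land effect attached to the part's skeletal mesh.
; Both event overrides call the parent implementation first.
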(event EventOnPawnJumped
  (|CallParentFunction)
  (Effects|Components|ParticleSystem|SetBoolParameter (CallFunction|GetIdleVFXComponent) "IsOnGround" false))

(event EventOnPawnLanded
  (|CallParentFunction)
  (Effects|Components|ParticleSystem|SetBoolParameter (CallFunction|GetIdleVFXComponent) "IsOnGround" true)
  (PartModifier|SpawnSystem|SpawnSystemAttachedPartModifier "/BRCosmetics/Accessories/FORT_Backpacks/Backpack_F_MED_Summer_Hotdog/Meshes/FX/NS_HotDog_Summer_Land.NS_HotDog_Summer_Land" (CustomCharacter|GetSkeletalMeshForPartOnPlayerPawn)))

(fn ConstructionScript
  (|CallParentFunction))
