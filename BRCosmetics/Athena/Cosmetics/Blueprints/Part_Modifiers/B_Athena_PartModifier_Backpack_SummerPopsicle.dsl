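; Cosmetic part modifier for the Summer Popsicle backpack that makes the popsicle appear to melt over time.
; After post-resurrection init, it casts the backpack mesh's anim instance to the popsicle AnimBP and binds to its material roughness dispatcher.
; The dispatched animation time is range-mapped to melt and frost material scalars on the part mesh.
; The same time drives idle VFX parameters, shifting fog to drips, shrinking the spawn box and thinning drips.
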(event EventOnPostResIn
  (|CallParentFunction)
  (bind _asf_med_summer_popsicle_pack_anim_bp (Utilities|Casting|CastToF_MED_Summer_Popsicle_Pack_AnimBP (Components|SkeletalMesh|GetAnimInstance (Customization|GetSkeletalMeshforPartType (CustomCharacter|GetAssociatedPlayerPawn) (CustomCharacter|GetCharacterPartType) :!class /Script/FortniteGame.FortPlayerPawn))))
  (Default|BindEventtoMaterialRoughness _asf_med_summer_popsicle_pack_anim_bp (EventDispatchers|CreateEvent :!event Custom|Roughness%20Update)))

(event Custom|Roughness%20Update
  (params (Time :type Float%20%28double-precision%29))
  (bind _returnvalue (Math|Float|MapRangeClamped Time 1.2e+02 2.7e+02 0.0 1.0))
  (bind _returnvalue_2 (CallFunction|GetIdleVFXComponent))
  (bind _returnvalue_3 (CustomCharacter|GetSkeletalMeshForPartOnPlayerPawn))
  ;[0 0 0] Use Animation Time Control to manage Roughness on Material of Popsicle and VFX
  (Rendering|Material|SetScalarParameterValueonMaterials _returnvalue_3 "SP_MeltedAmount" (Math|Float|MapRangeClamped Time 1.2e+02 4.2e+02 0.0 1.0))
  (Effects|Components|ParticleSystem|SetFloatParameter _returnvalue_2 "SpawnScale" (Math|Float|MapRangeClamped Time 1.2e+02 2.2e+02 1.0)) ; Swaps the Fog for Drips
  (Rendering|Material|SetScalarParameterValueonMaterials _returnvalue_3 "SP_MeltedAmount_Frost" _returnvalue)
  (Effects|Components|ParticleSystem|SetVectorParameter _returnvalue_2 "BoxSize" (Math|Vector|Lerp(Vector) "21.000000,5.000000,35.000000" (Math|Vector|Lerp(Vector) "15.000000,5.000000,30.000000" "10.000000,5.000000,10.000000" (Math|Float|MapRangeClamped Time 2.7e+02 4.2e+02 0.0 1.0)) _returnvalue))
  (Effects|Components|ParticleSystem|SetVectorParameter _returnvalue_2 "Box Offset" (Math|Vector|Lerp(Vector) "-1.000000,-1.000000,-1.000000" (Math|Vector|Lerp(Vector) "0.000000,-1.000000,-2.000000" "2.000000,-1.000000,-8.000000" (Math|Float|MapRangeClamped Time 2.7e+02 4.2e+02 0.0 1.0)) (Math|Float|MapRangeClamped Time 1.2e+02 2.7e+02 0.0 1.0)))
  (Effects|Components|ParticleSystem|SetFloatParameter _returnvalue_2 "SpawnSettledScale" (Math|Float|MapRangeClamped Time 2.7e+02 4.2e+02 1.0 0.5)) ; Reduces Drips on Shrink
)

(fn ConstructionScript
  (|CallParentFunction))
