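; Backpack cosmetic part modifier for SwipePlunk that reacts to player eliminations with material animation and VFX.
; In frontend, a looping timeline drives ReactivityPreview while a delayed, simulated kill spawns the elimination effect.
; In game, a kill-stat watcher activates idle VFX and timestamps the material animation with game time.
; Stat changes call the empty OnReactivity hook and spawn elimination Niagara with frontend and hidden-body-part flags, setting idle KillCount if parameter lookup fails.
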(event EventOnPostResIn
  (|CallParentFunction)
  (if (Reactive|Frontend|IsCosmeticObjectDisplayedinFrontendContext 0)
    (Utilities|FlowControl|Sequence
      (:then_0
        (bind (_direction _reactivity) (Timeline|FE_TAMaterial Play)
          (:Update
            (Rendering|Material|SetScalarParameterValueonMaterials (CustomCharacter|GetSkeletalMeshForPartOnPlayerPawn) "ReactivityPreview" _reactivity))
          (:Finished
            (Timeline|FE_TAMaterial PlayFromStart))))
      (:then_1
        (Utilities|FlowControl|Delay 1.0)
        (bind _result (Math|Integer|IncrementInt (Variables|Default|GetFEKillCount))
          (:%20%20
            (Utilities|FlowControl|Delay 1.0)
            (CallFunction|OnReactivity _result)
            (bind _returnvalue (CallFunction|GetVFXSpawnParameters "(TagName=\"CosmeticPart.VFX.Elimination\")")
              (:Success
                (bind _returnvalue_2 (PartModifier|SpawnSystem|SpawnSystemAttachedPartModifier (Utilities|Struct|BreakCharacterPartModifierVFXSpawnParameters _returnvalue) (.AttachToComponent (Utilities|Struct|BreakCharacterPartModifierVFXSpawnParameters _returnvalue)) (.SocketName (Utilities|Struct|BreakCharacterPartModifierVFXSpawnParameters _returnvalue)) "0, 0, 0" "0, 0, 0" (.LocationType (Utilities|Struct|BreakCharacterPartModifierVFXSpawnParameters _returnvalue)) true (.bAutoActivate (Utilities|Struct|BreakCharacterPartModifierVFXSpawnParameters _returnvalue)))
                  (:Success
                    (bind _output_get (Variables|Default|SetEliminationParticle _returnvalue_2))
                    (Niagara|SetNiagaraVariable(Bool) _output_get "bisFrontEnd" (Reactive|Frontend|IsCosmeticObjectDisplayedinFrontendContext 0))
                    (Utilities|FlowControl|Delay 8.0))))
              (:Failure
                _))
            (Effects|Components|ParticleSystem|SetIntParameter (CallFunction|GetVFXComponent "(TagName=\"CosmeticPart.VFX.Idle\")") "KillCount" _result)
            (Utilities|FlowControl|Delay 4.0)))))
    (else
      (CallFunction|PlayTAMaterial))))

(event EventOnStatChanged
  (params (State%20Name :type Name :default "None") (Stat%20Value :type Integer))
  (|CallParentFunction State%20Name Stat%20Value)
  (CallFunction|OnReactivity Stat%20Value)
  (bind _returnvalue (CallFunction|GetVFXSpawnParameters "(TagName=\"CosmeticPart.VFX.Elimination\")")
    (:Success
      (bind _returnvalue_2 (PartModifier|SpawnSystem|SpawnSystemAttachedPartModifier (Utilities|Struct|BreakCharacterPartModifierVFXSpawnParameters _returnvalue) (.AttachToComponent (Utilities|Struct|BreakCharacterPartModifierVFXSpawnParameters _returnvalue)) (.SocketName (Utilities|Struct|BreakCharacterPartModifierVFXSpawnParameters _returnvalue)) "0, 0, 0" "0, 0, 0" (.LocationType (Utilities|Struct|BreakCharacterPartModifierVFXSpawnParameters _returnvalue)) true (.bAutoActivate (Utilities|Struct|BreakCharacterPartModifierVFXSpawnParameters _returnvalue)))
        (:Success
          (bind _output_get (Variables|Default|SetEliminationParticle _returnvalue_2))
          (Niagara|SetNiagaraVariable(Bool) _output_get "bisFrontEnd" (Reactive|Frontend|IsCosmeticObjectDisplayedinFrontendContext 0))
          (CallFunction|SetSpawnedParticleHiddenTags _returnvalue_2))))
    (:Failure
      (Effects|Components|ParticleSystem|SetIntParameter (CallFunction|GetVFXComponent "(TagName=\"CosmeticPart.VFX.Idle\")") "KillCount" Stat%20Value))))

(event Custom|OnReactivity
  (params (KillCount :type Integer)))

(event Custom|Play%20TA%20Material
  ;[0.1418 0.05353 0.15 0.5] When killing an enemy is played animation
  (Cosmetics|WatchCosmeticStat (CustomCharacter|GetAssociatedPlayerPawn) "(TagName=\"StatManager.AthenaKills\")" 0 false
    (:OnStatChanged
      (Components|Activation|Activate (CallFunction|GetIdleVFXComponent) true)
      (Rendering|Material|SetScalarParameterValueonMaterials (CustomCharacter|GetSkeletalMeshForPartOnPlayerPawn) "AnimDuration" 4.0)
      (Rendering|Material|SetScalarParameterValueonMaterials (CustomCharacter|GetSkeletalMeshForPartOnPlayerPawn) "GameTimeFromBP" (Utilities|Time|GetGameTimeinSeconds)))))

(fn SetSpawnedParticleHiddenTags
  (params (NiagaraParticle :type Niagara%20Particle%20System%20Component%20Object%20Reference))
  (Niagara|SetNiagaraVariable(Bool) NiagaraParticle "bisHeadHidden" (Variables|HiddenPartCheck|GetbIsHeadHidden))
  (Niagara|SetNiagaraVariable(Bool) NiagaraParticle "bisHandHidden" (Variables|HiddenPartCheck|GetbIsHandHidden))
  (Niagara|SetNiagaraVariable(Bool) NiagaraParticle "bisFootHidden" (Variables|HiddenPartCheck|GetbIsFootHidden)))

(fn ConstructionScript
  (|CallParentFunction))
