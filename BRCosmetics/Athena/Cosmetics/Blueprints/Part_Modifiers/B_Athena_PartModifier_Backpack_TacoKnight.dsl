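; Backpack cosmetic part modifier for TacoKnight that feeds player movement into its idle particle effect.
; On each movement update, it sets the "RunningSpeed" particle parameter to the pawn's horizontal speed, or zero when stationary.
; On landing, it waits briefly, then plays a timeline from the start whose track drives the "Landing" particle parameter.
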(event EventOnPawnLanded
  (|CallParentFunction)
  (Utilities|FlowControl|Delay 0.05)
  (bind (_direction _spawntime) (Timeline|Timeline_0 PlayFromStart)
    (:Update
      (Effects|Components|ParticleSystem|SetFloatParameter (CallFunction|GetIdleVFXComponent) "Landing" _spawntime))))

(event EventOnCharacterMovementUpdated
  (params (DeltaSeconds :type Float%20%28single-precision%29) (OldLocation :type Vector :default "0, 0, 0") (OldVelocity :type Vector :default "0, 0, 0"))
  (bind _returnvalue (CustomCharacter|GetAssociatedPlayerPawn))
  (bind _returnvalue_2 (Transformation|GetVelocity _returnvalue))
  (bind _returnvalue_3 (CallFunction|GetIdleVFXComponent))
  (bind _returnvalue_4 (* _returnvalue_2 "1.000000,1.000000, 0"))
  (|CallParentFunction DeltaSeconds OldLocation OldVelocity)
  (if (Math|Vector|Equal(Vector) _returnvalue_4 "0, 0, 0" 0.0001)
    (Effects|Components|ParticleSystem|SetFloatParameter _returnvalue_3 "RunningSpeed" 0.0)
    (else
      (Effects|Components|ParticleSystem|SetFloatParameter _returnvalue_3 "RunningSpeed" (Math|Vector|VectorLength _returnvalue_4)))))

(fn ConstructionScript
  (|CallParentFunction))
