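; Part modifier for the Tar backpack cosmetic that drives its idle and stat-triggered hit reaction VFX.
; On post-res-in, it attaches the idle VFX component to the part's skeletal mesh at the "pack" socket and caches the active backpack Particle variant index.
; When a stat changes to a positive value, it spawns a pooled, auto-release hit reaction Niagara system on that socket, picking the default or variant effect from the cached index.
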(event EventOnPostResIn
  (|CallParentFunction)
  (Utilities|FlowControl|Sequence
    (:then_0
      (Transformation|AttachComponentToComponent (CallFunction|GetIdleVFXComponent) (CustomCharacter|GetSkeletalMeshForPartOnPlayerPawn) "pack" "KeepRelative" "KeepRelative" "KeepRelative" true))
    (:then_1
      (bind _index (CosmeticLoadout|GetVariantChannelActiveIndexforSlotActor "(TagName=\"LoadoutSlot.Backpack\")" (CustomCharacter|GetAssociatedPlayerPawn) "(TagName=\"Cosmetics.Variant.Channel.Particle\")" (Utilities|Array|MakeArray "(TagName=\"Cosmetics.Variant.Property.Particle1\")" "(TagName=\"Cosmetics.Variant.Property.Particle2\")")))
      (Variables|Default|SetCosmeticLoadoutInt _index))))

(event EventOnStatChanged
  (params (State%20Name :type Name :default "None") (Stat%20Value :type Integer))
  (|CallParentFunction State%20Name Stat%20Value)
  (if (> Stat%20Value 0)
    (PartModifier|SpawnSystem|SpawnSystemAttachedPartModifier (Utilities|Select :Index (Variables|Default|GetCosmeticLoadoutInt) :Option%200 "/BRCosmetics/Accessories/FORT_Backpacks/Backpack_M_MED_Tar/FX/NS_Backpack_Tar_HitReaction.NS_Backpack_Tar_HitReaction" :Option%201 "/BRCosmetics/Accessories/FORT_Backpacks/Backpack_M_MED_Tar/FX/NS_Backpack_Tar_HitReaction_Variant.NS_Backpack_Tar_HitReaction_Variant") (CustomCharacter|GetSkeletalMeshForPartOnPlayerPawn) "pack" "0, 0, 0" "0, 0, 0" "SnapToTarget" true true true "AutoRelease")))

(fn ConstructionScript
  (|CallParentFunction))
