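; Part modifier for the Target Practice backpack that displays the wearer's Athena kill count and reacts to changes in it.
; It watches the kill stat, pushing count and trigger time into mesh materials and mirroring slot 0 material parameters to the Niagara blinking light.
; Stat changes also play a sound and activate falling-paper particles, whose dynamic digit materials receive the previous count, with tens paper enabled on multiples of ten.
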(event EventOnPawnAndPartChanged
  (|CallParentFunction)
  (CallFunction|CreateCascadeMID)
  (bind _returnvalue (Class|FortCosmeticFunctionLibrary|GetCosmeticStatValue (CustomCharacter|GetAssociatedPlayerPawn) "(TagName=\"StatManager.AthenaKills\")"))
  (Variables|Default|SetKillCount (Math|Conversions|ToFloat(Integer) _returnvalue))
  (CallFunction|UpdateMaterialParameters)
  (bind (_statname _statvalue) (Cosmetics|WatchCosmeticStat (CustomCharacter|GetAssociatedPlayerPawn) "(TagName=\"StatManager.AthenaKills\")" 0 false)
    (:OnStatChanged
      (Variables|Default|SetKillCount (Math|Conversions|ToFloat(Integer) _statvalue))
      (Utilities|FlowControl|Sequence
        (:then_0
          (Variables|Default|SetTriggerGameTime (Utilities|Time|GetGameTimeinSeconds))
          (CallFunction|UpdateMaterialParameters))
        (:then_1
          (CallFunction|FXSetup&Activation))
        (:then_2
          (CallFunction|AudioSetup&Activation))))))

(event EventOnPostResIn
  (|CallParentFunction)
  (CallFunction|UpdateMaterialParameters))

(event Custom|FX%20Setup%20&%20Activation
  (bind _returnvalue (Utilities|Time|GetGameTimeinSeconds))
  (bind _returnvalue_2 (CustomCharacter|GetSkeletalMeshForPartOnPlayerPawn))
  (bind _killcount (Variables|Default|GetKillCount))
  (bind _returnvalue_3 (Math|Float|%25%28Float%29 _killcount 1e+01))
  (bind _returnvalue_4 (== _returnvalue_3 0.0))
  (bind _returnvalue_5 (> _killcount 1.0))
  (bind _killcount_2 (Variables|Default|GetKillCount))
  (bind _returnvalue_6 (- _killcount_2 1.0))
  (bind _killcount_3 (Variables|Default|GetKillCount))
  (bind _returnvalue_7 (!= _killcount_3 0.0))
  (bind _returnvalue_8 (and _returnvalue_5 _returnvalue_4))
  (bind _output (Utilities|ClampMinFloat _returnvalue_6))
  (bind _falling_paper (Variables|Default|GetFalling%20Paper))
  (bind _particalmat10sparam (Variables|Default|GetParticalMat10sParam))
  (bind _particalmat1sparam (Variables|Default|GetParticalMat1sParam))
  ;[1 1 1] Attachment
  (Transformation|AttachComponentToComponent (Variables|Default|GetFalling%20Paper) _returnvalue_2 "pack" "KeepRelative" "KeepRelative" "KeepRelative" true)
  (Transformation|AttachComponentToComponent (Variables|Default|GetBlinkingLight) _returnvalue_2 "pack" "KeepRelative" "KeepRelative" "KeepRelative" true)
  ;[1 1 1] FX-Shennanigans
  (Components|Activation|Activate _falling_paper true)
  (Effects|Components|ParticleSystem|SetEmitterEnable _falling_paper "10sPaper" _returnvalue_8)
  (Effects|Components|ParticleSystem|SetEmitterEnable _falling_paper "10sPaperGlow" _returnvalue_8)
  (Effects|Components|ParticleSystem|SetEmitterEnable _falling_paper "1sPaper" _returnvalue_7)
  (Effects|Components|ParticleSystem|SetEmitterEnable _falling_paper "1sPaperGlow" _returnvalue_7)
  ;;[1 1 1] Set Kill Count Parameter on Particle Material
  (Rendering|Material|SetScalarParameterValue _particalmat10sparam "TP_KillsInput" _output :!class /Script/Engine.MaterialInstanceDynamic)
  (Rendering|Material|SetScalarParameterValue _particalmat1sparam "TP_KillsInput" _output :!class /Script/Engine.MaterialInstanceDynamic)
  ;;[1 1 1] Feed Game Time for Number Glow on kill
  (Rendering|Material|SetScalarParameterValue _particalmat10sparam "TP_GameTimeFromBP" _returnvalue :!class /Script/Engine.MaterialInstanceDynamic)
  (Rendering|Material|SetScalarParameterValue _particalmat1sparam "TP_GameTimeFromBP" _returnvalue :!class /Script/Engine.MaterialInstanceDynamic))

(event Custom|Audio%20Setup%20&%20Activation
  ;[1 1 1] Audio-Shennanigans
  (if (> (Variables|Default|GetKillCount) 0.0)
    (Audio|SpawnSoundAttached (Variables|Default|GetOnKillSound) (CustomCharacter|GetSkeletalMeshForPartOnPlayerPawn))))

(fn CreateCascadeMID
  (bind _falling_paper (Variables|Default|GetFalling%20Paper))
  ;[1 1 1] Create Dynamic Material for Particles
  (bind _returnvalue (Rendering|Material|CreateNamedDynamicMaterialInstance _falling_paper "Ten" "/BRCosmetics/Accessories/FORT_Backpacks/Backpack_M_MED_Target_Practice/Materials/M_TargetPractice_FX.M_TargetPractice_FX"))
  (bind _output_get (Variables|Default|SetParticalMat10sParam _returnvalue))
  (Effects|Components|ParticleSystem|SetMaterialParameter _falling_paper "Ten" _output_get)
  (bind _returnvalue_2 (Rendering|Material|CreateNamedDynamicMaterialInstance _falling_paper "One" "/BRCosmetics/Accessories/FORT_Backpacks/Backpack_M_MED_Target_Practice/Materials/M_TargetPractice_FX.M_TargetPractice_FX"))
  (bind _output_get_2 (Variables|Default|SetParticalMat1sParam _returnvalue_2))
  (Effects|Components|ParticleSystem|SetMaterialParameter _falling_paper "One" _output_get_2))

(fn UpdateMaterialParameters
  (bind _returnvalue (CustomCharacter|GetSkeletalMeshForPartOnPlayerPawn))
  (bind _triggergametime (Variables|Default|GetTriggerGameTime))
  ;[1 1 1] Set Mesh Materials
  (Rendering|Material|SetScalarParameterValueonMaterials _returnvalue "TP_KillsInput" (Variables|Default|GetKillCount))
  (Rendering|Material|SetScalarParameterValueonMaterials _returnvalue "TP_GameTimeFromBP" _triggergametime)
  ;[1 1 1] Passing material instance values to the Niagara lights
  (bind _asmaterial_instance_dynamic (Utilities|Casting|CastToMaterialInstanceDynamic (Rendering|Material|GetMaterial _returnvalue 0)))
  (bind _returnvalue_2 (Rendering|Material|GetScalarParameterValue _asmaterial_instance_dynamic "TP_TimeOffset" :!class /Script/Engine.MaterialInstanceDynamic))
  (Effects|Components|ParticleSystem|SetFloatParameter (Variables|Default|GetBlinkingLight) "User.TP_TimeOffset" _returnvalue_2)
  (bind _returnvalue_3 (Rendering|Material|GetScalarParameterValue _asmaterial_instance_dynamic "TP_UseVertexColorChannelForTimeOFfset" :!class /Script/Engine.MaterialInstanceDynamic))
  (Effects|Components|ParticleSystem|SetFloatParameter (Variables|Default|GetBlinkingLight) "User.TP_UseVertexColorChannelForTimeOFfset" _returnvalue_3)
  (Effects|Components|ParticleSystem|SetFloatParameter (Variables|Default|GetBlinkingLight) "User.TP_GameTimeFromBP" _triggergametime)
  (bind _returnvalue_4 (Rendering|Material|GetScalarParameterValue _asmaterial_instance_dynamic "TP_ReactiveGlowDuration" :!class /Script/Engine.MaterialInstanceDynamic))
  (Effects|Components|ParticleSystem|SetFloatParameter (Variables|Default|GetBlinkingLight) "User.TP_ReactiveGlowDuration" _returnvalue_4)
  (bind _returnvalue_5 (Rendering|Material|GetScalarParameterValue _asmaterial_instance_dynamic "TP_ReactiveBlinkDuration" :!class /Script/Engine.MaterialInstanceDynamic))
  (Effects|Components|ParticleSystem|SetFloatParameter (Variables|Default|GetBlinkingLight) "User.TP_ReactiveBlinkDuration" _returnvalue_5)
  (bind _returnvalue_6 (Rendering|Material|GetScalarParameterValue _asmaterial_instance_dynamic "TP_LightsDefaultSpeed" :!class /Script/Engine.MaterialInstanceDynamic))
  (Effects|Components|ParticleSystem|SetFloatParameter (Variables|Default|GetBlinkingLight) "User.TP_LightsDefaultSpeed" _returnvalue_6)
  (bind _returnvalue_7 (Rendering|Material|GetScalarParameterValue _asmaterial_instance_dynamic "TP_LightsReactiveSpeed" :!class /Script/Engine.MaterialInstanceDynamic))
  (Effects|Components|ParticleSystem|SetFloatParameter (Variables|Default|GetBlinkingLight) "User.TP_LightsReactiveSpeed" _returnvalue_7)
  (bind _returnvalue_8 (Rendering|Material|GetScalarParameterValue _asmaterial_instance_dynamic "TP_ValueToHideonReactive" :!class /Script/Engine.MaterialInstanceDynamic))
  (Effects|Components|ParticleSystem|SetFloatParameter (Variables|Default|GetBlinkingLight) "User.TP_ValueToHideonReactive" _returnvalue_8))

(fn ConstructionScript)
