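; Backpack part modifier Blueprint for the TeriyakiPrincess back bling cosmetic, adjusting how its mesh renders.
; On the post-res-in event it calls the parent, then gets the skeletal mesh for its character part type from the associated player pawn.
; It raises that mesh's translucent sort priority to 1, affecting the draw order of its translucent materials.
; The construction script is empty.
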(event EventOnPostResIn
  (|CallParentFunction)
  (Rendering|SetTranslucentSortPriority (Customization|GetSkeletalMeshforPartType (CustomCharacter|GetAssociatedPlayerPawn) (CustomCharacter|GetCharacterPartType) :!class /Script/FortniteGame.FortPlayerPawn) 1))

(fn ConstructionScript)
