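; Part modifier for the toon Teriyaki Fish back bling, driving its Niagara effects for idle, landing and skydiving.
; After post-resource init, it reads the backpack's Progressive variant stage and treats index 1 as a dead fish, passed to the idle VFX as User.DeadFish.
; Landing spawns a bowl-socket effect with the same flag unless the modifier is hidden.
; Skydiving spawns another bowl effect, with a mesh validity retry, and deactivates it when skydiving ends.
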(event EventOnPostResIn
  (|CallParentFunction)
  (bind _index (CosmeticLoadout|GetVariantChannelActiveIndexforSlotActor "(TagName=\"LoadoutSlot.Backpack\")" (CustomCharacter|GetAssociatedPlayerPawn) "(TagName=\"Cosmetics.Variant.Channel.Progressive\")" (Utilities|Array|MakeArray "(TagName=\"Cosmetics.Variant.Property.Stage1\")" "(TagName=\"Cosmetics.Variant.Property.Stage2\")") false false))
  (bind _output_get (Variables|Default|SetCosmeticLoadoutIndex _index))
  (Variables|Default|SetIsDead (== _output_get 1))
  (bind _asniagara_particle_system_component (Utilities|Casting|CastToNiagaraComponent (CallFunction|GetIdleVFXComponent)))
  (Niagara|SetNiagaraVariable(Bool) _asniagara_particle_system_component "User.DeadFish" (Variables|Default|GetIsDead)))

(event EventOnPawnLanded
  (|CallParentFunction)
  (if (not (CallFunction|IsPartModifierHidden))
    (bind _returnvalue (PartModifier|SpawnSystem|SpawnSystemAttachedPartModifier "/BRCosmetics/Accessories/FORT_Backpacks/Backpack_M_MED_Teriyaki_Fish_Toon/FX/NS_Teriyaki_Fish_Toon_Pack_JumpLand.NS_Teriyaki_Fish_Toon_Pack_JumpLand" (CustomCharacter|GetSkeletalMeshForPartOnPlayerPawn) "FX_Bowl")
      (:Success
        (Niagara|SetNiagaraVariable(Bool) _returnvalue "User.DeadFish" (Variables|Default|GetIsDead))))))

(event EventOnBeginSkydiving
  (bind _returnvalue (CustomCharacter|GetSkeletalMeshForPartOnPlayerPawn))
  (|CallParentFunction)
  (Utilities|FlowControl|Sequence
    (:then_1
      (Utilities|IsValid _returnvalue
        (:Is%20Valid
          _)
        (:Is%20Not%20Valid
          (Utilities|FlowControl|Delay 0.1)
          (Utilities|IsValid (CustomCharacter|GetSkeletalMeshForPartOnPlayerPawn)
            (:Is%20Valid
              _))))
      (bind _returnvalue_2 (PartModifier|SpawnSystem|SpawnSystemAttachedPartModifier "/BRCosmetics/Accessories/FORT_Backpacks/Backpack_M_MED_Teriyaki_Fish_Toon/FX/NS_Teriyaki_Fish_Toon_Pack_Skydive.NS_Teriyaki_Fish_Toon_Pack_Skydive" _returnvalue "FX_Bowl")
        (:Success
          (Variables|Default|SetSkydiveFX _returnvalue_2))))))

(event EventOnEndSkydiving
  (bind _skydivefx (Variables|Default|GetSkydiveFX))
  (|CallParentFunction)
  (Utilities|IsValid _skydivefx
    (:Is%20Valid
      (Components|Activation|Deactivate _skydivefx))))

(fn ConstructionScript)
