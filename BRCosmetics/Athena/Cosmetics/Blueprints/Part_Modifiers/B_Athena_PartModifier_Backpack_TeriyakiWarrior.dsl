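; Part modifier for the Teriyaki Warrior backpack cosmetic that reacts visually to the wearer's eliminations.
; When the pawn or part changes, it watches the AthenaKills stat on the associated player pawn and stores each update as KillCount.
; For counts above zero, it attempts to attach an elimination particle effect at "pack" on the part's skeletal mesh.
; Whether or not that spawn succeeds, a timeline restarts and drives the materials' slurp outer emissive multiplier.
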(event EventOnPawnAndPartChanged
  (|CallParentFunction)
  (bind (_statname _statvalue) (Cosmetics|WatchCosmeticStat (CustomCharacter|GetAssociatedPlayerPawn) "(TagName=\"StatManager.AthenaKills\")" 0 false)
    (:OnStatChanged
      (bind _output_get (Variables|Default|SetKillCount (Math|Conversions|ToFloat(Integer) _statvalue)))
      (if (> _output_get 0.0)
        (PartModifier|SpawnSystem|SpawnEmitterAttachedPartModifier(Legacy) "/BRCosmetics/Accessories/FORT_Backpacks/Backpack_M_MED_Teriyaki_Warrior/FX/P_Teriyaki_Warrior_Backpack_Elimination.P_Teriyaki_Warrior_Backpack_Elimination" (CustomCharacter|GetSkeletalMeshForPartOnPlayerPawn) "pack"
          (:Success
            (bind (_direction _slurpemissivemultiplier) (Timeline|Timeline_0 PlayFromStart)
              (:Update
                (Rendering|Material|SetScalarParameterValueonMaterials (CustomCharacter|GetSkeletalMeshForPartOnPlayerPawn) "SlurpEmissiveOuterColorMultiplier" _slurpemissivemultiplier))))
          (:Failure
            (Timeline|Timeline_0 PlayFromStart)))))))

(fn ConstructionScript)
