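; Backpack part modifier for the Tide King cosmetic that reacts to the owning player's kill count.
; When the pawn or part changes, it watches the StatManager.AthenaKills cosmetic stat on the associated player pawn.
; On each stat change it writes the current game time into a scalar parameter on the dynamic material at slot 1 of the part's skeletal mesh, presumably driving a time-based material effect.
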(event EventOnPawnAndPartChanged
  (|CallParentFunction)
  (Cosmetics|WatchCosmeticStat (CustomCharacter|GetAssociatedPlayerPawn) "(TagName=\"StatManager.AthenaKills\")" 0 false
    (:OnStatChanged
      (Rendering|Material|SetScalarParameterValue (Class|BTechArtGlobalLibrary|GetDynamicMaterialInstance (CustomCharacter|GetSkeletalMeshForPartOnPlayerPawn) 1) "TK_GameTimeFromBP_CCPM" (Utilities|Time|GetGameTimeinSeconds) :!class /Script/Engine.MaterialInstanceDynamic))))

(fn ConstructionScript
  (|CallParentFunction))
