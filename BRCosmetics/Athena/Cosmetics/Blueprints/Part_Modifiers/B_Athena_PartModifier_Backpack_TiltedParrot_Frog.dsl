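; Cosmetic part modifier for the TiltedParrot Frog backpack that makes its materials react progressively to player eliminations.
; In the frontend it loops a timed preview; in game it watches the AthenaKills stat to trigger OnElimination.
; OnElimination writes previous count, clamped target and start time as material scalars for the animation.
; A timer tracks each animation, so an interrupting elimination starts from the interpolated current progress.
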(event Custom|RestartFrontendReactivityLoop
  ;[1 1 1] FE preview loop
  (Variables|Default|SetElims 0)
  (Utilities|FlowControl|Delay 1.0)
  (CallFunction|OnElimination (+ (Variables|Default|GetElims) 1))
  (Utilities|FlowControl|Delay 4.0)
  (if (>= (Variables|Default|GetElims) (Variables|Default|GetMaxElims))
    _
    (else
      _))
  ;[1 1 1] FE preview reset
  (Utilities|FlowControl|Delay 3.0)
  (Rendering|Material|SetScalarParameterValueonMaterials (CustomCharacter|GetSkeletalMeshForPartOnPlayerPawn) "PreviousEliminations_CCPM" 7.0)
  (Rendering|Material|SetScalarParameterValueonMaterials (CustomCharacter|GetSkeletalMeshForPartOnPlayerPawn) "Eliminations_CCPM" 8.0)
  (Rendering|Material|SetScalarParameterValueonMaterials (CustomCharacter|GetSkeletalMeshForPartOnPlayerPawn) "ReactiveStartTime_CCPM" (Utilities|Time|GetGameTimeinSeconds))
  (Utilities|FlowControl|Delay 6.0)
  (Rendering|Material|SetScalarParameterValueonMaterials (CustomCharacter|GetSkeletalMeshForPartOnPlayerPawn) "ReactiveStartTime_CCPM" -1e+02)
  (Rendering|Material|SetScalarParameterValueonMaterials (CustomCharacter|GetSkeletalMeshForPartOnPlayerPawn) "PreviousEliminations_CCPM" 0.0)
  (Rendering|Material|SetScalarParameterValueonMaterials (CustomCharacter|GetSkeletalMeshForPartOnPlayerPawn) "Eliminations_CCPM" 0.0)
  (CallFunction|RestartFrontendReactivityLoop))

(event EventOnPostResIn
  (|CallParentFunction)
  (if (Reactive|Frontend|ShouldDisplayFrontendReactivity 0 "Backpack")
    ;[1 1 1] FE preview loop
    (Variables|Default|SetElims 0)
    (Utilities|FlowControl|Delay 1.0)
    (CallFunction|OnElimination (+ (Variables|Default|GetElims) 1))
    (Utilities|FlowControl|Delay 4.0)
    (if (>= (Variables|Default|GetElims) (Variables|Default|GetMaxElims))
      _
      (else
        _))
    ;[1 1 1] FE preview reset
    (Utilities|FlowControl|Delay 3.0)
    (Rendering|Material|SetScalarParameterValueonMaterials (CustomCharacter|GetSkeletalMeshForPartOnPlayerPawn) "PreviousEliminations_CCPM" 7.0)
    (Rendering|Material|SetScalarParameterValueonMaterials (CustomCharacter|GetSkeletalMeshForPartOnPlayerPawn) "Eliminations_CCPM" 8.0)
    (Rendering|Material|SetScalarParameterValueonMaterials (CustomCharacter|GetSkeletalMeshForPartOnPlayerPawn) "ReactiveStartTime_CCPM" (Utilities|Time|GetGameTimeinSeconds))
    (Utilities|FlowControl|Delay 6.0)
    (Rendering|Material|SetScalarParameterValueonMaterials (CustomCharacter|GetSkeletalMeshForPartOnPlayerPawn) "ReactiveStartTime_CCPM" -1e+02)
    (Rendering|Material|SetScalarParameterValueonMaterials (CustomCharacter|GetSkeletalMeshForPartOnPlayerPawn) "PreviousEliminations_CCPM" 0.0)
    (Rendering|Material|SetScalarParameterValueonMaterials (CustomCharacter|GetSkeletalMeshForPartOnPlayerPawn) "Eliminations_CCPM" 0.0)
    (CallFunction|RestartFrontendReactivityLoop)
    (else
      ;[1 1 1] Init for ingame
      (bind (_statname _statvalue) (Cosmetics|WatchCosmeticStat (CustomCharacter|GetAssociatedPlayerPawn) "(TagName=\"StatManager.AthenaKills\")")
        (:OnStatChanged
          (CallFunction|OnElimination _statvalue))))))

(event Custom|OnElimination
  (params (NumEliminations :type Integer))
  ;[1 1 1] Prevent same-frame retriggering
  (if (!= NumEliminations (Variables|Default|GetElims))
    (Variables|Default|SetElims NumEliminations)
    (if (not (Variables|Default|GetAnimInProgress))
      ;[1 1 1] Set progressive reactivity anim start & end values
      (Variables|Default|SetPreviousProgress (Math|Conversions|ToFloat(Integer) (Math|Integer|Clamp(Integer) (- (Variables|Default|GetElims) 1) 0 (Variables|Default|GetMaxElims))))
      (else
        ;;[1 1 1] Handle subsequent elimination during active animation
        (Variables|Default|SetPreviousProgress (Math|Float|Lerp (Variables|Default|GetTargetProgress) (Variables|Default|GetPreviousProgress) (/ (Utilities|Time|GetTimerRemainingTimebyHandle (Variables|Default|GetReactivityAnimTimer)) (Variables|Default|GetReactivityAnimDuration))))
        (Utilities|Time|ClearandInvalidateTimerbyHandle (Variables|Default|GetReactivityAnimTimer))))
    (Variables|Default|SetTargetProgress (Math|Conversions|ToFloat(Integer) (Math|Integer|Clamp(Integer) (Variables|Default|GetElims) 0 (Variables|Default|GetMaxElims))))
    ;[1 1 1] Update material parameters
    (Rendering|Material|SetScalarParameterValueonMaterials (CustomCharacter|GetSkeletalMeshForPartOnPlayerPawn) "PreviousEliminations_CCPM" (Variables|Default|GetPreviousProgress))
    (Rendering|Material|SetScalarParameterValueonMaterials (CustomCharacter|GetSkeletalMeshForPartOnPlayerPawn) "Eliminations_CCPM" (Variables|Default|GetTargetProgress))
    (Rendering|Material|SetScalarParameterValueonMaterials (CustomCharacter|GetSkeletalMeshForPartOnPlayerPawn) "ReactiveStartTime_CCPM" (Utilities|Time|GetGameTimeinSeconds))
    (Variables|Default|SetAnimInProgress true)
    (bind _returnvalue (Utilities|Time|SetTimerbyEvent (delegate Custom|OnReactivityAnimFinished) (Variables|Default|GetReactivityAnimDuration)))
    (Variables|Default|SetReactivityAnimTimer _returnvalue)))

(event Custom|OnReactivityAnimFinished
  (Variables|Default|SetAnimInProgress false))

(fn UpdateAudioStatValue
  (params (AudioComponent :type Array%20of%20Audio%20Component%20Object%20References) (InName :type Name :default "None") (StatValue :type Integer))
  (for _array_element AudioComponent
    (Audio|Parameter|SetIntegerParameter _array_element InName StatValue :!class /Script/AudioExtensions.AudioParameterControllerInterface)
    (return)))

(fn ConstructionScript
  (|CallParentFunction))
