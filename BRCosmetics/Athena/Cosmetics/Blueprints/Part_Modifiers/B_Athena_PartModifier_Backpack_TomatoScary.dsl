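; Part modifier for the Tomato Scary back bling, which reacts to the player's kill count.
; Cosmetic stat stages pick one of four faces through material offsets and update idle Niagara liquid color and intensity.
; Kill changes pass count and game time to the material for red spread and spawn straw-tip FX, as does landing.
; When frontend reactivity is enabled, a timed DoN sequence previews the kill stages.
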(event EventOnPawnLanded
  (|CallParentFunction)
  (bind _returnvalue (PartModifier|SpawnSystem|SpawnSystemAttachedPartModifier "/BRCosmetics/Accessories/FORT_Backpacks/Backpack_M_MED_Tomato_Scary/FX/NS_Backpack_TomatoScary_JumpedFX.NS_Backpack_TomatoScary_JumpedFX" (CustomCharacter|GetSkeletalMeshForPartOnPlayerPawn) "FX_StrawEnd")
    (:Success
      (FX|FXSetNSParams _returnvalue))))

(event EventOnCosmeticStateActivated
  (params (Stat%20Name :type Name :default "None") (Stat%20Value :type Integer) (STAGE%20Num%20Being%20Activated :type Integer))
  (bind _returnvalue (CustomCharacter|GetSkeletalMeshForPartOnPlayerPawn))
  (Cosmetics|WatchCosmeticStat (CustomCharacter|GetAssociatedPlayerPawn) "(TagName=\"StatManager.AthenaKills\")" 0 false)
  (bind _data (Utilities|FlowControl|MergeExecutionwithData STAGE%20Num%20Being%20Activated (Variables|Default|GetCurrentKillCount))
    (:Then
      (switch int _data
        (:0
          ;[1 1 1] Facial Expressions
          (Rendering|Material|SetScalarParameterValueonMaterials _returnvalue "TS_Face_X" 0.0)
          (Rendering|Material|SetScalarParameterValueonMaterials _returnvalue "TS_Face_Y" 0.0)
          (FX|FXUpdateNiagaraParams))
        (:1
          (Rendering|Material|SetScalarParameterValueonMaterials _returnvalue "TS_Face_X" 0.5)
          (Rendering|Material|SetScalarParameterValueonMaterials _returnvalue "TS_Face_Y" 0.0)
          (FX|FXUpdateNiagaraParams))
        (:2
          (Rendering|Material|SetScalarParameterValueonMaterials _returnvalue "TS_Face_X" 0.0)
          (Rendering|Material|SetScalarParameterValueonMaterials _returnvalue "TS_Face_Y" 0.5)
          (FX|FXUpdateNiagaraParams))
        (:3
          (Rendering|Material|SetScalarParameterValueonMaterials _returnvalue "TS_Face_X" 0.5)
          (Rendering|Material|SetScalarParameterValueonMaterials _returnvalue "TS_Face_Y" 0.5)
          (FX|FXUpdateNiagaraParams))))))

(event Custom|OnFrontendCycleEvent
  (params (FrontendKillIndex :type Integer))
  (bind _returnvalue (CustomCharacter|GetSkeletalMeshForPartOnPlayerPawn))
  (bind _returnvalue_2 (CustomCharacter|GetSkeletalMeshForPartOnPlayerPawn))
  (Utilities|FlowControl|Sequence
    (:then_0
      (Variables|Default|SetCurrentKillCount FrontendKillIndex))
    (:then_1
      (bind _data (Utilities|FlowControl|MergeExecutionwithData (stale EventOnCosmeticStateActivated STAGE%20Num%20Being%20Activated) (Variables|Default|GetCurrentKillCount))
        (:Then
          (switch int _data
            (:0
              ;[1 1 1] Facial Expressions
              (Rendering|Material|SetScalarParameterValueonMaterials _returnvalue "TS_Face_X" 0.0)
              (Rendering|Material|SetScalarParameterValueonMaterials _returnvalue "TS_Face_Y" 0.0)
              (FX|FXUpdateNiagaraParams))
            (:1
              (Rendering|Material|SetScalarParameterValueonMaterials _returnvalue "TS_Face_X" 0.5)
              (Rendering|Material|SetScalarParameterValueonMaterials _returnvalue "TS_Face_Y" 0.0)
              (FX|FXUpdateNiagaraParams))
            (:2
              (Rendering|Material|SetScalarParameterValueonMaterials _returnvalue "TS_Face_X" 0.0)
              (Rendering|Material|SetScalarParameterValueonMaterials _returnvalue "TS_Face_Y" 0.5)
              (FX|FXUpdateNiagaraParams))
            (:3
              (Rendering|Material|SetScalarParameterValueonMaterials _returnvalue "TS_Face_X" 0.5)
              (Rendering|Material|SetScalarParameterValueonMaterials _returnvalue "TS_Face_Y" 0.5)
              (FX|FXUpdateNiagaraParams))))))
    (:then_2
      ;[1 1 1] SpreadRed
      (Rendering|Material|SetScalarParameterValueonMaterials _returnvalue_2 "TS_KillCount" (Math|Conversions|ToFloat(Integer) (stale EventOnStatChanged _output_get)))
      (Rendering|Material|SetScalarParameterValueonMaterials _returnvalue_2 "TS_GameTimeFromBP" (Utilities|Time|GetGameTimeinSeconds))
      (Cosmetics|WatchCosmeticStat (CustomCharacter|GetAssociatedPlayerPawn) "(TagName=\"StatManager.AthenaKills\")" 0 false
        (:OnStatChanged
          (bind _returnvalue_3 (PartModifier|SpawnSystem|SpawnSystemAttachedPartModifier "/BRCosmetics/Accessories/FORT_Backpacks/Backpack_M_MED_Tomato_Scary/FX/NS_Backpack_TomatoScary_ReactFX.NS_Backpack_TomatoScary_ReactFX" (CustomCharacter|GetSkeletalMeshForPartOnPlayerPawn) "FX_StrawEnd")
            (:Success
              (FX|FXSetNSParams _returnvalue_3))))))))

(event EventOnStatChanged
  (params (State%20Name :type Name :default "None") (Stat%20Value :type Integer))
  (bind _returnvalue (CustomCharacter|GetSkeletalMeshForPartOnPlayerPawn))
  (bind _output_get (Variables|Default|SetCurrentKillCount Stat%20Value))
  ;[1 1 1] SpreadRed
  (Rendering|Material|SetScalarParameterValueonMaterials _returnvalue "TS_KillCount" (Math|Conversions|ToFloat(Integer) _output_get))
  (Rendering|Material|SetScalarParameterValueonMaterials _returnvalue "TS_GameTimeFromBP" (Utilities|Time|GetGameTimeinSeconds))
  (Cosmetics|WatchCosmeticStat (CustomCharacter|GetAssociatedPlayerPawn) "(TagName=\"StatManager.AthenaKills\")" 0 false
    (:OnStatChanged
      (bind _returnvalue_2 (PartModifier|SpawnSystem|SpawnSystemAttachedPartModifier "/BRCosmetics/Accessories/FORT_Backpacks/Backpack_M_MED_Tomato_Scary/FX/NS_Backpack_TomatoScary_ReactFX.NS_Backpack_TomatoScary_ReactFX" (CustomCharacter|GetSkeletalMeshForPartOnPlayerPawn) "FX_StrawEnd")
        (:Success
          (FX|FXSetNSParams _returnvalue_2))))))

(event EventOnPostResIn
  (|CallParentFunction)
  (if (Reactive|Frontend|ShouldDisplayFrontendReactivity 0 "Backpack")
    (Utilities|FlowControl|Delay 2.0)
    (bind _counter (Utilities|FlowControl|DoN 4)
      (:Exit
        (switch int _counter
          (:0
            _)
          (:1
            _)
          (:2
            _)
          (:3
            _)
          (:4
            (CallFunction|OnFrontendCycleEvent 0)
            (Utilities|FlowControl|Delay 3.0)
            (Utilities|FlowControl|Sequence
              (:then_0
                _)
              (:then_1
                _))))
        (CallFunction|OnFrontendCycleEvent _counter)
        (Utilities|FlowControl|RetriggerableDelay 3.0)))))

(fn FX_GetCurrentLiquidColor
  (return (Math|Color|SelectColor "(R=0.411458,G=0.068631,B=0.000000,A=1.000000)" (Utilities|Select :Index (Variables|Default|GetCurrentKillCount) :Option%200 "(R=0.163900,G=0.213542,B=0.012627,A=1.000000)" :Option%201 "(R=0.286458,G=0.171340,B=0.004101,A=1.000000)" :Option%202 "(R=0.348958,G=0.122125,B=0.004311,A=1.000000)" :Option%203 "(R=0.411458,G=0.068631,B=0.000000,A=1.000000)") (>= (Variables|Default|GetCurrentKillCount) 3))))

(fn FX_Set_NS_Params
  (params (InNiagaraSys :type Niagara%20Particle%20System%20Component%20Object%20Reference))
  (Niagara|SetNiagaraVariable(LinearColor) InNiagaraSys "User.LiquidColor" (FX|FXGetCurrentLiquidColor)))

(fn FX_UpdateNiagaraParams
  (bind _returnvalue (CallFunction|GetIdleVFXComponent))
  (bind _asniagara_particle_system_component (Utilities|Casting|CastToNiagaraComponent _returnvalue))
  (Niagara|SetNiagaraVariable(LinearColor) _asniagara_particle_system_component "User.LiquidColor" (FX|FXGetCurrentLiquidColor))
  (Niagara|SetNiagaraVariable(Float) _asniagara_particle_system_component "User.KillRamp" (Math|Float|MapRangeClamped (Math|Conversions|ToFloat(Integer) (Variables|Default|GetCurrentKillCount)) 0.0 3.0 0.0 1.0)))

(fn ConstructionScript
  (|CallParentFunction))
