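; Backpack cosmetic part modifier for Track Waver that makes its idle particle effect react to flight through an "IsFlying" float parameter.
; Jumping, skydiving start (one tick later) and any non-zero vertical pawn velocity checked each Tick set the parameter to 1.
; Landing and skydive end play a timeline that drives its fade.
; When frontend reactivity is enabled, two timelines chained by delays loop the parameter as a preview.
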(event EventOnPostResIn
  (bind _returnvalue (CallFunction|GetIdleVFXComponent))
  ; Front-End
  (|CallParentFunction)
  (if (CallFunction|ShouldDisplayFrontendReactivity)
    (Utilities|FlowControl|Delay 4.0)
    (bind (_direction _fade_fe) (Timeline|Timeline_FE_1 PlayFromStart)
      (:Update
        (Effects|Components|ParticleSystem|SetFloatParameter _returnvalue "IsFlying" _fade_fe)
        (Utilities|FlowControl|Delay 4.0)
        (bind (_direction_2 _fade_fe_2) (Timeline|Timeline_FE_2 PlayFromStart)
          (:Update
            (Effects|Components|ParticleSystem|SetFloatParameter _returnvalue "IsFlying" _fade_fe_2)
            (Utilities|FlowControl|Delay 8.0)
            (Timeline|Timeline_FE_1 PlayFromStart)))))))

(event EventTick
  (params (DeltaSeconds :type Float%20%28single-precision%29))
  (|CallParentFunction DeltaSeconds)
  (bind _output_get (Variables|Default|SetPawnVelocity (* (Transformation|GetVelocity (CustomCharacter|GetAssociatedPlayerPawn)) "0.000000,0.000000,1.000000")))
  (if (> (Math|Vector|VectorLength _output_get) 0)
    (Effects|Components|ParticleSystem|SetFloatParameter (CallFunction|GetIdleVFXComponent) "IsFlying" 1.0)))

(event EventOnBeginSkydiving
  (Utilities|FlowControl|DelayUntilNextTick)
  (Effects|Components|ParticleSystem|SetFloatParameter (CallFunction|GetIdleVFXComponent) "IsFlying" 1.0))

(event EventOnPawnJumped
  (|CallParentFunction)
  (Effects|Components|ParticleSystem|SetFloatParameter (CallFunction|GetIdleVFXComponent) "IsFlying" 1.0))

(event EventOnPawnLanded
  (|CallParentFunction)
  (bind (_direction _fade) (Timeline|Timeline PlayFromStart)
    (:Update
      (Effects|Components|ParticleSystem|SetFloatParameter (CallFunction|GetIdleVFXComponent) "IsFlying" _fade))))

(event EventOnEndSkydiving
  (Timeline|Timeline PlayFromStart))

(fn ConstructionScript
  (|CallParentFunction))
