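; Backpack part modifier for the Twenty Two cosmetic that plays its on-kill Niagara burst at the backpack's FX_Center socket.
; A stat change with a positive value spawns the burst and destroys it after a retriggerable delay, with DoOnce and a short delay meant to absorb duplicate triggers.
; In the front end, a looping timer replays the burst as a flagged preview until the modifier is destroyed.
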(event Custom|CE_Frontend%20FX%20Loop%20Burst
  (bind _returnvalue (PartModifier|SpawnSystem|SpawnSystemAttachedPartModifier "/BRCosmetics/Accessories/FORT_Backpacks/Backpack_F_MED_Twenty_Two/FX/NS_Backpack_Twenty_Two_On_Kill.NS_Backpack_Twenty_Two_On_Kill" (CustomCharacter|GetSkeletalMeshForPartOnPlayerPawn) "FX_Center")
    (:Success
      (Effects|Components|ParticleSystem|SetBoolParameter _returnvalue "bisFrontEndPreview" (CustomCosmeticModifier|IsinFrontEnd)))))

(event EventOnPostResIn
  (|CallParentFunction)
  (if (CustomCosmeticModifier|IsinFrontEnd)
    (Utilities|FlowControl|DoOnce false
      (:Completed
        (bind _returnvalue (Utilities|Time|SetTimerbyEvent (delegate Custom|CE_Frontend%20FX%20Loop%20Burst) 6.0 true))
        (Variables|Default|SetTimer _returnvalue)))))

(event EventDestroyed
  (if (CustomCosmeticModifier|IsinFrontEnd)
    (Utilities|Time|ClearandInvalidateTimerbyHandle (Variables|Default|GetTimer))))

(event EventOnStatChanged
  (params (State%20Name :type Name :default "None") (Stat%20Value :type Integer))
  (|CallParentFunction State%20Name Stat%20Value)
  (Utilities|FlowControl|Sequence
    (:then_0
      _)
    (:then_1
      ;[1 1 1] Fix the issue that sometimes this event is triggered multiple times
      (Utilities|FlowControl|Delay 0.5)))
  (Utilities|FlowControl|DoOnce false
    (:Completed
      (if (> Stat%20Value 0)
        (bind _returnvalue (PartModifier|SpawnSystem|SpawnSystemAttachedPartModifier "/BRCosmetics/Accessories/FORT_Backpacks/Backpack_F_MED_Twenty_Two/FX/NS_Backpack_Twenty_Two_On_Kill.NS_Backpack_Twenty_Two_On_Kill" (CustomCharacter|GetSkeletalMeshForPartOnPlayerPawn) "FX_Center")
          (:Success
            (Utilities|FlowControl|RetriggerableDelay 3.0)
            (Components|DestroyComponent _returnvalue)))))))

(fn ConstructionScript
  (|CallParentFunction))
