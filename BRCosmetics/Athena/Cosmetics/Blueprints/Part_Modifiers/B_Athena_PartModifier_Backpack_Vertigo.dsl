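; Part modifier for the Vertigo backpack cosmetic that plays a reactive effect on stat changes and previews it in the front end.
; A stat change spawns the OnKill Niagara system on the backpack root and, on success, plays the reduceSize timeline, which drives the mesh's "WPO Offset" material parameter.
; In the front end, the FE_Trigger timeline fires the same reaction, tinting and dimming the spawned system, or the idle VFX if spawning fails.
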(event EventOnPostResIn
  (bind _returnvalue (CallFunction|GetIdleVFXComponent))
  (|CallParentFunction)
  (if (CustomCosmeticModifier|IsinFrontEnd)
    (Timeline|FE_Trigger PlayFromStart
      (:Trigger
        (bind _returnvalue_2 (PartModifier|SpawnSystem|SpawnSystemAttachedPartModifier "/BRCosmetics/Accessories/FORT_Backpacks/Backpack_M_MED_Vertigo/FX/NS_Backpackl_Vertigo_OnKill.NS_Backpackl_Vertigo_OnKill" (CustomCharacter|GetSkeletalMeshForPartOnPlayerPawn) "root" "0, 0, 0" "0, 0, 0" "SnapToTarget")
          (:Success
            (Effects|Components|ParticleSystem|SetColorParameter _returnvalue_2 "Color_Tint" "(R=0.036458,G=0.023356,B=0.023356,A=1.000000)")
            (Effects|Components|ParticleSystem|SetFloatParameter _returnvalue_2 "EmissiveIntensity" 0.35))
          (:Failure
            (Effects|Components|ParticleSystem|SetColorParameter _returnvalue "Color_Tint" "(R=0.036458,G=0.023356,B=0.023356,A=1.000000)")
            (Effects|Components|ParticleSystem|SetFloatParameter _returnvalue "EmissiveIntensity" 0.35)))
        (CallFunction|FrontEnd)))))

(event Custom|FrontEnd
  (bind (_direction _newtrack_0) (Timeline|reduceSize PlayFromStart)
    (:Update
      (Rendering|Material|SetScalarParameterValueonMaterials (CustomCharacter|GetSkeletalMeshForPartOnPlayerPawn) "WPO Offset" (Math|Float|Lerp 0.0 -0.5 _newtrack_0)))))

(event EventOnStatChanged
  (params (State%20Name :type Name :default "None") (Stat%20Value :type Integer))
  (|CallParentFunction State%20Name Stat%20Value)
  (PartModifier|SpawnSystem|SpawnSystemAttachedPartModifier "/BRCosmetics/Accessories/FORT_Backpacks/Backpack_M_MED_Vertigo/FX/NS_Backpackl_Vertigo_OnKill.NS_Backpackl_Vertigo_OnKill" (CustomCharacter|GetSkeletalMeshForPartOnPlayerPawn) "root" "0, 0, 0" "0, 0, 0" "SnapToTarget"
    (:Success
      (Timeline|reduceSize PlayFromStart))))

(fn ConstructionScript
  (|CallParentFunction))
