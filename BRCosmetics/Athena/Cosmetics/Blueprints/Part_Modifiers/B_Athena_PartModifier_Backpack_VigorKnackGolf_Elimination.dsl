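; Backpack cosmetic part modifier for VigorKnackGolf that makes the back bling's idle VFX react to eliminations.
; Stat changes pass their value to KillReactivity, which writes positive counts to the idle particle component as KillCount and stamps ReactivityTime with game time.
; When frontend reactivity should display, post-res-in resets a fake kill count, then after a short delay increments it through a timed stage loop to preview the effect.
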(event EventOnPostResIn
  ; FE Reactivity
  (|CallParentFunction)
  (if (CallFunction|ShouldDisplayFrontendReactivity)
    (Variables|Default|SetFEKillCount 0)
    (Utilities|FlowControl|Delay 1.0)
    ; FE Stage Loop
    (bind _result (Math|Integer|IncrementInt (Variables|Default|GetFEKillCount))
      (:%20%20
        (CallFunction|KillReactivity _result)
        (Utilities|FlowControl|Delay 3.75)))))

(event Custom|KillReactivity
  (params (KillCount :type Integer))
  (bind _returnvalue (CallFunction|GetIdleVFXComponent))
  ; Elim Reactivity
  (if (> KillCount 0)
    (Effects|Components|ParticleSystem|SetIntParameter _returnvalue "KillCount" KillCount)
    (Effects|Components|ParticleSystem|SetFloatParameter _returnvalue "ReactivityTime" (Utilities|Time|GetGameTimeinSeconds))))

(event EventOnStatChanged
  (params (State%20Name :type Name :default "None") (Stat%20Value :type Integer))
  ; Elim Reactivity
  (|CallParentFunction State%20Name Stat%20Value)
  (CallFunction|KillReactivity Stat%20Value))

(fn ConstructionScript
  (|CallParentFunction))
