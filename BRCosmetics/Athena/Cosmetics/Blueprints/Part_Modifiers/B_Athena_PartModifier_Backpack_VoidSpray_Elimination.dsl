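; Backpack part modifier for the Void Spray cosmetic that makes its materials and Niagara VFX react to the player's elimination count.
; Positive stat changes stamp the material's Timestamp, apply KillCount after a short delay, and spawn elimination VFX or set the idle VFX count if spawn parameters are unavailable.
; A restartable three-second timer makes rapid eliminations reinitialize the existing effect instead of spawning another.
; In the frontend, an incremented fake kill count previews the reaction.
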(event EventOnPostResIn
  (|CallParentFunction)
  (if (CallFunction|ShouldDisplayFrontendReactivity)
    (Utilities|FlowControl|Delay 1.0)
    (bind _result (Math|Integer|IncrementInt (Variables|Default|GetFEKillCount))
      (:%20%20
        (CallFunction|OnReactivity _result)
        (bind _returnvalue (CallFunction|GetVFXSpawnParameters "(TagName=\"CosmeticPart.VFX.Elimination\")")
          (:Success
            (CallFunction|SpawnVFX _returnvalue _result)
            (Utilities|FlowControl|Delay 8.0))
          (:Failure
            _))
        (Effects|Components|ParticleSystem|SetIntParameter (CallFunction|GetVFXComponent "(TagName=\"CosmeticPart.VFX.Idle\")") "KillCount" _result)
        (Utilities|FlowControl|Delay 8.0)))))

(event Custom|Disable%20FXLoop
  ; 3 Second Lockout from spawning VFX
  (Variables|Default|SetFXLoopActive false))

(event EventOnStatChanged
  (params (State%20Name :type Name :default "None") (Stat%20Value :type Integer))
  (bind _eliminationparticle (Variables|Default|GetEliminationParticle))
  (bind _timerhandle (Variables|Default|GetTimerHandle))
  (|CallParentFunction State%20Name Stat%20Value)
  (if (> Stat%20Value 0)
    (CallFunction|OnReactivity Stat%20Value)
    (if (Utilities|Time|IsValidTimerHandle _timerhandle)
      (Utilities|Time|ClearandInvalidateTimerbyHandle _timerhandle))
    (if (Variables|Default|GetFXLoopActive)
      ; Reset existing VFX
      (Niagara|ReinitializeSystem _eliminationparticle)
      (Niagara|SetNiagaraVariable(Int32) _eliminationparticle "KillCount" Stat%20Value)
      (Variables|Default|SetFXLoopActive true)
      (else
        ; Spawn New VFX
        (bind _returnvalue (CallFunction|GetVFXSpawnParameters "(TagName=\"CosmeticPart.VFX.Elimination\")")
          (:Success
            (CallFunction|SpawnVFX _returnvalue Stat%20Value)
            (Variables|Default|SetFXLoopActive true))
          (:Failure
            (Effects|Components|ParticleSystem|SetIntParameter (CallFunction|GetVFXComponent "(TagName=\"CosmeticPart.VFX.Idle\")") "KillCount" Stat%20Value)))))
    ; 3 Second Lockout from spawning VFX
    (bind _returnvalue_2 (Utilities|Time|SetTimerbyEvent (delegate Custom|Disable%20FXLoop) 3.0))
    (Variables|Default|SetTimerHandle _returnvalue_2)))

(event Custom|Spawn%20VFX
  (params (Part%20Modifier%20Params :type Character%20Part%20Modifier%20VFXSpawn%20Parameters%20Structure) (KillCount :type Integer))
  (bind (_systemtemplate _attachtocomponent _socketname _locationtype _bautoactivate) (Utilities|Struct|BreakCharacterPartModifierVFXSpawnParameters Part%20Modifier%20Params))
  ; Set VFX Reactivity
  (bind _returnvalue (PartModifier|SpawnSystem|SpawnSystemAttachedPartModifier _systemtemplate _attachtocomponent _socketname "0, 0, 0" "0, 0, 0" _locationtype true _bautoactivate)
    (:Success
      (bind _output_get (Variables|Default|SetEliminationParticle _returnvalue))
      (Niagara|SetNiagaraVariable(Int32) _output_get "KillCount" KillCount)
      (Niagara|SetNiagaraVariable(Bool) _output_get "bisFrontEnd" (Reactive|Frontend|IsCosmeticObjectDisplayedinFrontendContext 0)))))

(event Custom|OnReactivity
  (params (KillCount :type Integer))
  (bind _returnvalue (CustomCharacter|GetSkeletalMeshforPartonOwner))
  ; Set Material Reactivity
  (Rendering|Material|SetScalarParameterValueonMaterials _returnvalue "Timestamp" (Utilities|Time|GetGameTimeinSeconds))
  (Utilities|FlowControl|Delay 0.9)
  (Rendering|Material|SetScalarParameterValueonMaterials _returnvalue "KillCount" (Math|Conversions|ToFloat(Integer) KillCount)))

(fn ConstructionScript
  (|CallParentFunction))
