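; Backpack cosmetic part modifier for WavySignalBeef that plays a sound in reaction to the owning player's kills.
; After post-res-in, if ShouldDisplayFrontendReactivity passes for "Backpack", it watches the AthenaKills stat and sets ReactivityEnabled from IsinFrontEnd.
; Each stat change calls OnKill, which, when ReactivityEnabled is true, spawns the Apprentice backbling elimination cue attached to the part's skeletal mesh.
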(event Custom|OnKill
  ;[1 1 1] On Kill Event
  (if (Variables|Default|GetReactivityEnabled)
    (Default|SetandPlaySound)))

(event EventOnPostResIn
  (|CallParentFunction)
  (if (Reactive|Frontend|ShouldDisplayFrontendReactivity 0 "Backpack")
    (Utilities|FlowControl|Sequence
      (:then_0
        ;[1 1 1] Setup On Kill
        (Cosmetics|WatchCosmeticStat (CustomCharacter|GetAssociatedPlayerPawn) "(TagName=\"StatManager.AthenaKills\")" 0 false
          (:OnStatChanged
            (CallFunction|OnKill))))
      (:then_1
        ;[1 1 1] Read CIDs
        (Variables|Default|SetReactivityEnabled (CustomCosmeticModifier|IsinFrontEnd))))))

(fn SetAndPlaySound
  (Audio|SpawnSoundAttached "/BRCosmetics/Sounds/Backblings/Apprentice/BackBling_Apprentice_Elimination_Open_Cue.BackBling_Apprentice_Elimination_Open_Cue" (CustomCharacter|GetSkeletalMeshForPartOnPlayerPawn)))

(fn ConstructionScript
  (|CallParentFunction))
