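; Part modifier for the Wayfare backpack that plays a Niagara visibility effect when the pack disappears and reappears during specific emotes.
; When an emote in EmotesToListenFor begins, a looping timer at world delta seconds polls IsPartModifierHidden, then hands off to a second timer watching for reappearance.
; DoOnce nodes guard each effect, spawned at the wayfare_body socket location.
; Ending the emote clears both timers.
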(event EventOnBeginEmote
  (params (MontageItemDef :type Fort%20Item%20Definition%20Object%20Reference) (PawnEmoting :type Fort%20Pawn%20Object%20Reference))
  (|CallParentFunction MontageItemDef PawnEmoting)
  (if (CallFunction|IsEmotetoListenTo MontageItemDef)
    (Utilities|FlowControl|Sequence
      (:then_0
        (bind _returnvalue (Utilities|Time|SetTimerbyFunctionName 0 "CheckPackDisappeared" (Utilities|Time|GetWorldDeltaSeconds) true))
        (Variables|Default|SetEmote_DisappearTimer _returnvalue))
      (:then_1
        (Utilities|FlowControl|Gate true
          (:Exit
            (if (CallFunction|IsPartModifierHidden)
              (Utilities|FlowControl|DoOnce false
                (:Completed
                  (Utilities|FlowControl|Sequence
                    (:then_0
                      (Utilities|Time|ClearandInvalidateTimerbyHandle (Variables|Default|GetEmote_DisappearTimer))
                      (CallFunction|SpawnVisibilityEffect))
                    (:then_1
                      ;[0.9861 0 1 0.302] Now start checking if its reappeared
                      (bind _returnvalue_2 (Utilities|Time|SetTimerbyFunctionName 0 "CheckPackReappeared" (Utilities|Time|GetWorldDeltaSeconds) true))
                      (Variables|Default|SetEmote_ReappearTimer _returnvalue_2)
                      ;[0.881 1 0 0.176] Reappearance
                      (Utilities|FlowControl|DoOnce false
                        (:Completed
                          (CallFunction|SpawnVisibilityEffect)
                          (Utilities|Time|ClearandInvalidateTimerbyHandle (Variables|Default|GetEmote_ReappearTimer)))))))))))))))

(event Custom|CheckPackDisappeared
  (Utilities|FlowControl|Gate true
    (:Exit
      (if (CallFunction|IsPartModifierHidden)
        (Utilities|FlowControl|DoOnce false
          (:Completed
            (Utilities|FlowControl|Sequence
              (:then_0
                (Utilities|Time|ClearandInvalidateTimerbyHandle (Variables|Default|GetEmote_DisappearTimer))
                (CallFunction|SpawnVisibilityEffect))
              (:then_1
                ;[0.9861 0 1 0.302] Now start checking if its reappeared
                (bind _returnvalue (Utilities|Time|SetTimerbyFunctionName 0 "CheckPackReappeared" (Utilities|Time|GetWorldDeltaSeconds) true))
                (Variables|Default|SetEmote_ReappearTimer _returnvalue)
                ;[0.881 1 0 0.176] Reappearance
                (Utilities|FlowControl|DoOnce false
                  (:Completed
                    (CallFunction|SpawnVisibilityEffect)
                    (Utilities|Time|ClearandInvalidateTimerbyHandle (Variables|Default|GetEmote_ReappearTimer))))))))))))

(event EventOnEndEmote
  (params (MontageItemDef :type Fort%20Item%20Definition%20Object%20Reference) (PawnEmoting :type Fort%20Pawn%20Object%20Reference))
  (|CallParentFunction MontageItemDef PawnEmoting)
  (if (CallFunction|IsEmotetoListenTo MontageItemDef)
    (Utilities|FlowControl|Sequence
      (:then_0
        (Utilities|FlowControl|Sequence
          (:then_0
            (Utilities|FlowControl|Gate true
              (:Exit
                (if (CallFunction|IsPartModifierHidden)
                  _))))
          (:then_1
            (Utilities|Time|ClearandInvalidateTimerbyHandle (Variables|Default|GetEmote_DisappearTimer))))
        (Utilities|FlowControl|DoOnce false
          (:Completed
            (Utilities|FlowControl|Sequence
              (:then_0
                (Utilities|Time|ClearandInvalidateTimerbyHandle (Variables|Default|GetEmote_DisappearTimer))
                (CallFunction|SpawnVisibilityEffect))
              (:then_1
                ;[0.9861 0 1 0.302] Now start checking if its reappeared
                (bind _returnvalue (Utilities|Time|SetTimerbyFunctionName 0 "CheckPackReappeared" (Utilities|Time|GetWorldDeltaSeconds) true))
                (Variables|Default|SetEmote_ReappearTimer _returnvalue)
                ;[0.881 1 0 0.176] Reappearance
                (Utilities|FlowControl|DoOnce false
                  (:Completed
                    (CallFunction|SpawnVisibilityEffect)
                    (Utilities|Time|ClearandInvalidateTimerbyHandle (Variables|Default|GetEmote_ReappearTimer)))))))))
      (:then_1
        (Utilities|Time|ClearandInvalidateTimerbyHandle (Variables|Default|GetEmote_ReappearTimer))))))

(event Custom|CheckPackReappeared
  ;[0.881 1 0 0.176] Reappearance
  (if (not (CallFunction|IsPartModifierHidden))
    (Utilities|FlowControl|DoOnce false
      (:Completed
        (CallFunction|SpawnVisibilityEffect)
        (Utilities|Time|ClearandInvalidateTimerbyHandle (Variables|Default|GetEmote_ReappearTimer))))))

(event Custom|SpawnVisibilityEffect
  (Niagara|SpawnSystematLocation "/BRCosmetics/Animation/Game/MainPlayer/Emotes/WayFare/FX/NS_Emote_WayfarePack_VisibilityChange.NS_Emote_WayfarePack_VisibilityChange" (Transformation|GetSocketLocation (CustomCharacter|GetSkeletalMeshForPartOnPlayerPawn) "wayfare_body")))

(fn IsEmoteToListenTo
  (params (ItemToFind :type Fort%20Item%20Definition%20Object%20Reference%20%28by%20ref%29))
  (return (!= (Utilities|Array|FindItem (Variables|Default|GetEmotesToListenFor) ItemToFind) -1)))

(fn ConstructionScript
  (|CallParentFunction))
