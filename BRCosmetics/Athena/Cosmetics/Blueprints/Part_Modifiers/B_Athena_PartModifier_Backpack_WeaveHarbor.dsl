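; Backpack part modifier for the WeaveHarbor cosmetic that changes its material and idle VFX with the wearer's health.
; Watched health, or a timed demo loop in the frontend, maps to three phases, with DBNO forcing the lowest.
; Phase changes play timed out/in material transitions and Niagara bursts (skipped for frontend holograms), locked to a minimum interval, then re-evaluated to prevent overlap.
; Landing briefly pulses the idle VFX LandBurst parameter.
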(event EventOnPawnLanded
  (bind _returnvalue (CallFunction|GetIdleVFXComponent))
  (|CallParentFunction)
  (Effects|Components|ParticleSystem|SetFloatParameter _returnvalue "LandBurst" 1.0)
  (Utilities|FlowControl|Delay 0.1)
  (Effects|Components|ParticleSystem|SetFloatParameter _returnvalue "LandBurst" 0.0))

(event EventOnPostResIn
  (bind _returnvalue (Utilities|Time|GetGameTimeinSeconds))
  (bind _previoushealthphase (Variables|Default|GetPreviousHealthPhase))
  (bind _returnvalue_2 (Math|Conversions|ToFloat(Integer) _previoushealthphase))
  (bind _transitionoutduration (Variables|Default|GetTransitionOutDuration))
  (bind _returnvalue_3 (+ _transitionoutduration _returnvalue))
  (|CallParentFunction)
  (if (CustomCosmeticModifier|IsinFrontEnd)
    ;[1 1 1] Frontend loop
    (Utilities|FlowControl|Delay 6.0)
    (bind _output_get (Variables|Default|SetFrontendLoopCounter (Math|Integer|%25%28Integer%29 (+ (Variables|Default|GetFrontendLoopCounter) 1) 3)))
    (Variables|Default|SetCurrentHealthPercentage (Utilities|Select :Index _output_get :Option%200 0.45 :Option%201 0.2 :Option%202 1.0))
    (Utilities|FlowControl|Sequence
      (:then_0
        _)
      (:then_1
        (CallFunction|CycleFrontendLoop)))
    (else
      (bind _returnvalue_4 (Fortnite|Utilities|AttemptGetPlayerStateforPawn (CustomCharacter|GetAssociatedPlayerPawn)))
      (bind _asfort_player_state_athena (Utilities|Casting|CastToFortPlayerStateAthena _returnvalue_4))
      (bind (_health _maxhealth) (Cosmetics|WatchPlayerHealth _asfort_player_state_athena)
        (:then
          ;[1 1 1] Set initial values to full health, to correctly initialize on non-squad clients
          (bind _output_get_2 (Variables|Default|SetCurrentHealthPercentage 1.0)) ; Assume 100% health on res-in to set correct value on non-squad remote players
          (Rendering|Material|SetScalarParameterValueonMaterials (CustomCharacter|GetSkeletalMeshForPartOnPlayerPawn) "HealthPhase" (Math|Conversions|ToFloat(Integer) (CallFunction|GetCurrentHealthPhase _output_get_2)))
          (Effects|Components|ParticleSystem|SetFloatParameter (CallFunction|GetIdleVFXComponent) "HealthPercent" (Variables|Default|GetCurrentHealthPercentage))
          (Effects|Components|ParticleSystem|SetFloatParameter (CallFunction|GetIdleVFXComponent) "HealthPhase" 0.0))
        (:OnHitPointsChanged
          (Variables|Default|SetCurrentHealthPercentage (Math|Float|SafeDivide _health _maxhealth))))))
  (Effects|Components|ParticleSystem|SetFloatParameter (CallFunction|GetIdleVFXComponent) "HealthPercent" (Variables|Default|GetCurrentHealthPercentage))
  ;[1 1 1] Check if health threshold has been crossed and we should enter a material transition
  (if (not (Variables|Default|GetIsTransitioningPhase))
    (Variables|Default|SetPreviousHealthPhase (Variables|Default|GetCurrentHealthPhase))
    (bind _output_get_3 (Variables|Default|SetCurrentHealthPhase (Math|Integer|SelectInt 2 (CallFunction|GetCurrentHealthPhase (Variables|Default|GetCurrentHealthPercentage)) (Health|IsDBNO (CustomCharacter|GetAssociatedPlayerPawn)))))
    (if (!= _output_get_3 (Variables|Default|GetPreviousHealthPhase))
      (Variables|Default|SetIsTransitioningPhase true)
      (Utilities|FlowControl|Sequence
        (:then_0
          ;[0.04866 0.15 0.06207 0.5] Added to Fix FORT-893416
          (if (not (Variables|/Script/FortniteGame.FortPlayerPawn|GetbIsInFrontEndHologram (CustomCharacter|GetAssociatedPlayerPawn)))
            (Niagara|SpawnSystemAttached "/BRCosmetics/Accessories/FORT_Backpacks/Backpack_M_MED_WeaveHarbor/FX/NS_Backpack_WeaveHarbor_Reactivity_In.NS_Backpack_WeaveHarbor_Reactivity_In" (CustomCharacter|GetSkeletalMeshForPartOnPlayerPawn) "base_main" "0, 0, 0" "0, 0, 0" "KeepRelativeOffset" true true "AutoRelease")))
        (:then_1
          ;[1 1 1] Out animation of previous texture
          (Effects|Components|ParticleSystem|SetFloatParameter (CallFunction|GetIdleVFXComponent) "HealthPhase" _returnvalue_2)
          (Rendering|Material|SetScalarParameterValueonMaterials (CustomCharacter|GetSkeletalMeshForPartOnPlayerPawn) "HealthPhase" _returnvalue_2)
          (Rendering|Material|SetScalarParameterValueonMaterials (CustomCharacter|GetSkeletalMeshForPartOnPlayerPawn) "TransitionOutStart" _returnvalue)
          (Rendering|Material|SetScalarParameterValueonMaterials (CustomCharacter|GetSkeletalMeshForPartOnPlayerPawn) "TransitionOutDuration" _transitionoutduration)
          (Effects|Components|ParticleSystem|SetFloatParameter (CallFunction|GetIdleVFXComponent) "TextFadeStartTime" _returnvalue_3)
          (Rendering|Material|SetScalarParameterValueonMaterials (CustomCharacter|GetSkeletalMeshForPartOnPlayerPawn) "TransitionInStart" _returnvalue_3)
          (Rendering|Material|SetScalarParameterValueonMaterials (CustomCharacter|GetSkeletalMeshForPartOnPlayerPawn) "TransitionInDuration" (Variables|Default|GetTransitionInDuration))
          (Utilities|FlowControl|Delay (Variables|Default|GetTransitionOutDuration))
          (Utilities|FlowControl|Sequence
            (:then_0
              ;[0.04866 0.15 0.06207 0.5] Added to Fix FORT-893416
              (if (not (Variables|/Script/FortniteGame.FortPlayerPawn|GetbIsInFrontEndHologram (CustomCharacter|GetAssociatedPlayerPawn)))
                (Niagara|SpawnSystemAttached "/BRCosmetics/Accessories/FORT_Backpacks/Backpack_M_MED_WeaveHarbor/FX/NS_Backpack_WeaveHarbor_Reactivity_Out.NS_Backpack_WeaveHarbor_Reactivity_Out" (CustomCharacter|GetSkeletalMeshForPartOnPlayerPawn) "base_main" "0.000000,0.000000,3.000000" "0, 0, 0" "KeepRelativeOffset" true true "AutoRelease")))
            (:then_1
              ;[1 1 1] In Animation of new texture
              (bind _output_get_4 (Variables|Default|SetCurrentHealthPhase (Math|Integer|SelectInt 2 (CallFunction|GetCurrentHealthPhase (Variables|Default|GetCurrentHealthPercentage)) (Health|IsDBNO (CustomCharacter|GetAssociatedPlayerPawn)))))
              (Rendering|Material|SetScalarParameterValueonMaterials (CustomCharacter|GetSkeletalMeshForPartOnPlayerPawn) "HealthPhase" (Math|Conversions|ToFloat(Integer) _output_get_4))
              (Utilities|FlowControl|Delay (Variables|Default|GetTransitionInDuration))
              ;[1 1 1] Fill up minimum interval before next transition in order to prevent concurrent playing of multiple cloud imprint animations on the Idle FX material
              (Utilities|FlowControl|Delay (Utilities|Operators|Subtract 4.0 (Variables|Default|GetTransitionInDuration) (Variables|Default|GetTransitionOutDuration)))
              (Variables|Default|SetIsTransitioningPhase false)
              (CallFunction|ReevaluateHealthPhase))))))))

(event Custom|Reevaluate%20Health%20Phase
  (bind _returnvalue (Utilities|Time|GetGameTimeinSeconds))
  (bind _previoushealthphase (Variables|Default|GetPreviousHealthPhase))
  (bind _returnvalue_2 (Math|Conversions|ToFloat(Integer) _previoushealthphase))
  (bind _transitionoutduration (Variables|Default|GetTransitionOutDuration))
  (bind _returnvalue_3 (+ _transitionoutduration _returnvalue))
  ;[1 1 1] Check if health threshold has been crossed and we should enter a material transition
  (if (not (Variables|Default|GetIsTransitioningPhase))
    (Variables|Default|SetPreviousHealthPhase (Variables|Default|GetCurrentHealthPhase))
    (bind _output_get (Variables|Default|SetCurrentHealthPhase (Math|Integer|SelectInt 2 (CallFunction|GetCurrentHealthPhase (Variables|Default|GetCurrentHealthPercentage)) (Health|IsDBNO (CustomCharacter|GetAssociatedPlayerPawn)))))
    (if (!= _output_get (Variables|Default|GetPreviousHealthPhase))
      (Variables|Default|SetIsTransitioningPhase true)
      (Utilities|FlowControl|Sequence
        (:then_0
          ;[0.04866 0.15 0.06207 0.5] Added to Fix FORT-893416
          (if (not (Variables|/Script/FortniteGame.FortPlayerPawn|GetbIsInFrontEndHologram (CustomCharacter|GetAssociatedPlayerPawn)))
            (Niagara|SpawnSystemAttached "/BRCosmetics/Accessories/FORT_Backpacks/Backpack_M_MED_WeaveHarbor/FX/NS_Backpack_WeaveHarbor_Reactivity_In.NS_Backpack_WeaveHarbor_Reactivity_In" (CustomCharacter|GetSkeletalMeshForPartOnPlayerPawn) "base_main" "0, 0, 0" "0, 0, 0" "KeepRelativeOffset" true true "AutoRelease")))
        (:then_1
          ;[1 1 1] Out animation of previous texture
          (Effects|Components|ParticleSystem|SetFloatParameter (CallFunction|GetIdleVFXComponent) "HealthPhase" _returnvalue_2)
          (Rendering|Material|SetScalarParameterValueonMaterials (CustomCharacter|GetSkeletalMeshForPartOnPlayerPawn) "HealthPhase" _returnvalue_2)
          (Rendering|Material|SetScalarParameterValueonMaterials (CustomCharacter|GetSkeletalMeshForPartOnPlayerPawn) "TransitionOutStart" _returnvalue)
          (Rendering|Material|SetScalarParameterValueonMaterials (CustomCharacter|GetSkeletalMeshForPartOnPlayerPawn) "TransitionOutDuration" _transitionoutduration)
          (Effects|Components|ParticleSystem|SetFloatParameter (CallFunction|GetIdleVFXComponent) "TextFadeStartTime" _returnvalue_3)
          (Rendering|Material|SetScalarParameterValueonMaterials (CustomCharacter|GetSkeletalMeshForPartOnPlayerPawn) "TransitionInStart" _returnvalue_3)
          (Rendering|Material|SetScalarParameterValueonMaterials (CustomCharacter|GetSkeletalMeshForPartOnPlayerPawn) "TransitionInDuration" (Variables|Default|GetTransitionInDuration))
          (Utilities|FlowControl|Delay (Variables|Default|GetTransitionOutDuration))
          (Utilities|FlowControl|Sequence
            (:then_0
              ;[0.04866 0.15 0.06207 0.5] Added to Fix FORT-893416
              (if (not (Variables|/Script/FortniteGame.FortPlayerPawn|GetbIsInFrontEndHologram (CustomCharacter|GetAssociatedPlayerPawn)))
                (Niagara|SpawnSystemAttached "/BRCosmetics/Accessories/FORT_Backpacks/Backpack_M_MED_WeaveHarbor/FX/NS_Backpack_WeaveHarbor_Reactivity_Out.NS_Backpack_WeaveHarbor_Reactivity_Out" (CustomCharacter|GetSkeletalMeshForPartOnPlayerPawn) "base_main" "0.000000,0.000000,3.000000" "0, 0, 0" "KeepRelativeOffset" true true "AutoRelease")))
            (:then_1
              ;[1 1 1] In Animation of new texture
              (bind _output_get_2 (Variables|Default|SetCurrentHealthPhase (Math|Integer|SelectInt 2 (CallFunction|GetCurrentHealthPhase (Variables|Default|GetCurrentHealthPercentage)) (Health|IsDBNO (CustomCharacter|GetAssociatedPlayerPawn)))))
              (Rendering|Material|SetScalarParameterValueonMaterials (CustomCharacter|GetSkeletalMeshForPartOnPlayerPawn) "HealthPhase" (Math|Conversions|ToFloat(Integer) _output_get_2))
              (Utilities|FlowControl|Delay (Variables|Default|GetTransitionInDuration))
              ;[1 1 1] Fill up minimum interval before next transition in order to prevent concurrent playing of multiple cloud imprint animations on the Idle FX material
              (Utilities|FlowControl|Delay (Utilities|Operators|Subtract 4.0 (Variables|Default|GetTransitionInDuration) (Variables|Default|GetTransitionOutDuration)))
              (Variables|Default|SetIsTransitioningPhase false)
              (CallFunction|ReevaluateHealthPhase))))))))

(event Custom|CycleFrontendLoop
  (bind _returnvalue (Utilities|Time|GetGameTimeinSeconds))
  (bind _previoushealthphase (Variables|Default|GetPreviousHealthPhase))
  (bind _returnvalue_2 (Math|Conversions|ToFloat(Integer) _previoushealthphase))
  (bind _transitionoutduration (Variables|Default|GetTransitionOutDuration))
  (bind _returnvalue_3 (+ _transitionoutduration _returnvalue))
  ;[1 1 1] Frontend loop
  (Utilities|FlowControl|Delay 6.0)
  (bind _output_get (Variables|Default|SetFrontendLoopCounter (Math|Integer|%25%28Integer%29 (+ (Variables|Default|GetFrontendLoopCounter) 1) 3)))
  (Variables|Default|SetCurrentHealthPercentage (Utilities|Select :Index _output_get :Option%200 0.45 :Option%201 0.2 :Option%202 1.0))
  (Utilities|FlowControl|Sequence
    (:then_0
      (Effects|Components|ParticleSystem|SetFloatParameter (CallFunction|GetIdleVFXComponent) "HealthPercent" (Variables|Default|GetCurrentHealthPercentage))
      ;[1 1 1] Check if health threshold has been crossed and we should enter a material transition
      (if (not (Variables|Default|GetIsTransitioningPhase))
        (Variables|Default|SetPreviousHealthPhase (Variables|Default|GetCurrentHealthPhase))
        (bind _output_get_2 (Variables|Default|SetCurrentHealthPhase (Math|Integer|SelectInt 2 (CallFunction|GetCurrentHealthPhase (Variables|Default|GetCurrentHealthPercentage)) (Health|IsDBNO (CustomCharacter|GetAssociatedPlayerPawn)))))
        (if (!= _output_get_2 (Variables|Default|GetPreviousHealthPhase))
          (Variables|Default|SetIsTransitioningPhase true)
          (Utilities|FlowControl|Sequence
            (:then_0
              ;[0.04866 0.15 0.06207 0.5] Added to Fix FORT-893416
              (if (not (Variables|/Script/FortniteGame.FortPlayerPawn|GetbIsInFrontEndHologram (CustomCharacter|GetAssociatedPlayerPawn)))
                (Niagara|SpawnSystemAttached "/BRCosmetics/Accessories/FORT_Backpacks/Backpack_M_MED_WeaveHarbor/FX/NS_Backpack_WeaveHarbor_Reactivity_In.NS_Backpack_WeaveHarbor_Reactivity_In" (CustomCharacter|GetSkeletalMeshForPartOnPlayerPawn) "base_main" "0, 0, 0" "0, 0, 0" "KeepRelativeOffset" true true "AutoRelease")))
            (:then_1
              ;[1 1 1] Out animation of previous texture
              (Effects|Components|ParticleSystem|SetFloatParameter (CallFunction|GetIdleVFXComponent) "HealthPhase" _returnvalue_2)
              (Rendering|Material|SetScalarParameterValueonMaterials (CustomCharacter|GetSkeletalMeshForPartOnPlayerPawn) "HealthPhase" _returnvalue_2)
              (Rendering|Material|SetScalarParameterValueonMaterials (CustomCharacter|GetSkeletalMeshForPartOnPlayerPawn) "TransitionOutStart" _returnvalue)
              (Rendering|Material|SetScalarParameterValueonMaterials (CustomCharacter|GetSkeletalMeshForPartOnPlayerPawn) "TransitionOutDuration" _transitionoutduration)
              (Effects|Components|ParticleSystem|SetFloatParameter (CallFunction|GetIdleVFXComponent) "TextFadeStartTime" _returnvalue_3)
              (Rendering|Material|SetScalarParameterValueonMaterials (CustomCharacter|GetSkeletalMeshForPartOnPlayerPawn) "TransitionInStart" _returnvalue_3)
              (Rendering|Material|SetScalarParameterValueonMaterials (CustomCharacter|GetSkeletalMeshForPartOnPlayerPawn) "TransitionInDuration" (Variables|Default|GetTransitionInDuration))
              (Utilities|FlowControl|Delay (Variables|Default|GetTransitionOutDuration))
              (Utilities|FlowControl|Sequence
                (:then_0
                  ;[0.04866 0.15 0.06207 0.5] Added to Fix FORT-893416
                  (if (not (Variables|/Script/FortniteGame.FortPlayerPawn|GetbIsInFrontEndHologram (CustomCharacter|GetAssociatedPlayerPawn)))
                    (Niagara|SpawnSystemAttached "/BRCosmetics/Accessories/FORT_Backpacks/Backpack_M_MED_WeaveHarbor/FX/NS_Backpack_WeaveHarbor_Reactivity_Out.NS_Backpack_WeaveHarbor_Reactivity_Out" (CustomCharacter|GetSkeletalMeshForPartOnPlayerPawn) "base_main" "0.000000,0.000000,3.000000" "0, 0, 0" "KeepRelativeOffset" true true "AutoRelease")))
                (:then_1
                  ;[1 1 1] In Animation of new texture
                  (bind _output_get_3 (Variables|Default|SetCurrentHealthPhase (Math|Integer|SelectInt 2 (CallFunction|GetCurrentHealthPhase (Variables|Default|GetCurrentHealthPercentage)) (Health|IsDBNO (CustomCharacter|GetAssociatedPlayerPawn)))))
                  (Rendering|Material|SetScalarParameterValueonMaterials (CustomCharacter|GetSkeletalMeshForPartOnPlayerPawn) "HealthPhase" (Math|Conversions|ToFloat(Integer) _output_get_3))
                  (Utilities|FlowControl|Delay (Variables|Default|GetTransitionInDuration))
                  ;[1 1 1] Fill up minimum interval before next transition in order to prevent concurrent playing of multiple cloud imprint animations on the Idle FX material
                  (Utilities|FlowControl|Delay (Utilities|Operators|Subtract 4.0 (Variables|Default|GetTransitionInDuration) (Variables|Default|GetTransitionOutDuration)))
                  (Variables|Default|SetIsTransitioningPhase false)
                  (CallFunction|ReevaluateHealthPhase))))))))
    (:then_1
      (CallFunction|CycleFrontendLoop))))

(event EventPlayerStartDBNO
  (|CallParentFunction)
  (CallFunction|ReevaluateHealthPhase)
  (EventDispatchers|BindEventtoOnDBNOStateChanged (CustomCharacter|GetAssociatedPlayerPawn) (delegate Custom|OnDBNOStateChanged)))

(event Custom|OnDBNOStateChanged
  (params (FortPawn :type Fort%20Pawn%20Object%20Reference) (bInIsDBNO :type Boolean))
  (CallFunction|ReevaluateHealthPhase))

(fn GetCurrentHealthPhase
  (params (Current%20Health%20Percentage :type Float%20%28double-precision%29))
  (if (> Current%20Health%20Percentage (Variables|Default|GetPhase2HealthThreshold))
    (return 0)
    (elif (> Current%20Health%20Percentage (Variables|Default|GetPhase3HealthThreshold))
      (return 1)
      (else
        (return 2)))))

(fn ConstructionScript
  (|CallParentFunction))
