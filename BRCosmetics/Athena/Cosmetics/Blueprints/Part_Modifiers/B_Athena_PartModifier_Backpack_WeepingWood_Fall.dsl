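; Backpack part modifier for the Weeping Woods Fall back bling that adds its Niagara effect to the cosmetic.
; When the owning pawn jumps, it runs the parent handler, then spawns the Weeping Woods Fall backpack Niagara system attached to the skeletal mesh of that part on the player pawn.
; The construction script is empty.
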(event EventOnPawnJumped
  (|CallParentFunction)
  (PartModifier|SpawnSystem|SpawnSystemAttachedPartModifier "/BRCosmetics/Accessories/FORT_Backpacks/Backpack_M_MED_Weeping_Woods_Fall/FX/NS_WeepingWoods_Fall_Backpack.NS_WeepingWoods_Fall_Backpack" (CustomCharacter|GetSkeletalMeshForPartOnPlayerPawn)))

(fn ConstructionScript)
