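; Cosmetic part modifier for the Weeping Woods Spring backpack that plays a particle effect when the player pawn jumps.
; The jump event calls the parent handler, then spawns the backpack's emitter through the legacy attached part-modifier spawn function.
; The emitter attaches to the skeletal mesh looked up for the character part on the player pawn.
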(event EventOnPawnJumped
  (|CallParentFunction)
  (PartModifier|SpawnSystem|SpawnEmitterAttachedPartModifier(Legacy) "/BRCosmetics/Accessories/FORT_Backpacks/Backpack_M_MED_Weeping_Woods_Spring/FX/P_Backpack_WeepingWoods_Spring.P_Backpack_WeepingWoods_Spring" (CustomCharacter|GetSkeletalMeshForPartOnPlayerPawn)))

(fn ConstructionScript)
