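; Cosmetic part modifier for the Winged Fury backpack that attaches its left and right wing components to the wings of the backpack mesh.
; On the post-res-in event it runs the parent logic and looks up the skeletal mesh for this part type on the associated player pawn.
; Each component snaps to the matching mid-wing dynamic bone in location and rotation, keeping relative scale.
; The construction script is empty.
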(event EventOnPostResIn
  (bind _returnvalue (CustomCharacter|GetCharacterPartType))
  (bind _returnvalue_2 (CustomCharacter|GetAssociatedPlayerPawn))
  (bind _returnvalue_3 (Customization|GetSkeletalMeshforPartType _returnvalue_2 _returnvalue :!class /Script/FortniteGame.FortPlayerPawn))
  (|CallParentFunction)
  (Transformation|AttachComponentToComponent (Variables|Default|GetP_WingedFuryBackpack_WingL) _returnvalue_3 "dyn_wing_mid_03_l" "SnapToTarget" "SnapToTarget" "KeepRelative" true)
  (Transformation|AttachComponentToComponent (Variables|Default|GetP_WingedFuryBackpack_WingR) _returnvalue_3 "dyn_wing_mid_03_r" "SnapToTarget" "SnapToTarget" "KeepRelative" true))

(fn ConstructionScript)
