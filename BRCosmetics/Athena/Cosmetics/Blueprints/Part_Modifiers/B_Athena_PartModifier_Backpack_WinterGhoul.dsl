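; Backpack part modifier Blueprint for the WinterGhoul cosmetic, adjusting how the backpack renders in the front end.
; On the post-res-in event it calls the parent, then checks whether it is running in the front end.
; If so, it sets the translucent sort priority of the part's skeletal mesh on the player pawn to 1.
; The construction script is empty.
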(event EventOnPostResIn
  (|CallParentFunction)
  (if (CustomCosmeticModifier|IsinFrontEnd)
    (Rendering|SetTranslucentSortPriority (CustomCharacter|GetSkeletalMeshForPartOnPlayerPawn) 1)))

(fn ConstructionScript)
