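; Backpack part modifier for the WiryPerk cosmetic that plays a reactive material animation when the player's kill stat changes.
; After resources initialize, it watches AthenaKills, or in frontend reactivity display fires after a short delay and on a looping timer.
; Each trigger restarts a timeline driving the TextAnimationA scalar on the part mesh's materials.
; PlayReactiveSound is empty, and nothing here calls SetAndPlaySound, which spawns an Apprentice backbling sound.
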(event EventOnPostResIn
  (|CallParentFunction)
  ; Setup On Kill
  (if (Reactive|Frontend|ShouldDisplayFrontendReactivity 0 "Backpack")
    (Utilities|FlowControl|Delay 4.0)
    (CallFunction|Onkill)
    (Utilities|Time|SetTimerbyEvent (delegate Custom|Onkill) 15.0 true)
    (else
      (Cosmetics|WatchCosmeticStat (CustomCharacter|GetAssociatedPlayerPawn) "(TagName=\"StatManager.AthenaKills\")" 0 false
        (:OnStatChanged
          (CallFunction|Onkill))))))

(event Custom|Onkill
  ; Setup On Kill
  (bind (_direction _animationa) (Timeline|Timeline PlayFromStart)
    (:Update
      (Rendering|Material|SetScalarParameterValueonMaterials (CustomCharacter|GetSkeletalMeshForPartOnPlayerPawn) "TextAnimationA" _animationa))))

(event Custom|PlayReactiveSound
  (params (Index :type Integer)))

(fn SetAndPlaySound
  (Audio|SpawnSoundAttached "/BRCosmetics/Sounds/Backblings/Apprentice/BackBling_Apprentice_Elimination_Open_Cue.BackBling_Apprentice_Elimination_Open_Cue" (CustomCharacter|GetSkeletalMeshForPartOnPlayerPawn)))

(fn ConstructionScript
  (|CallParentFunction))
