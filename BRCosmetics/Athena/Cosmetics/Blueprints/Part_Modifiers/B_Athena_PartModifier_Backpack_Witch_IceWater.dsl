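; Backpack part modifier Blueprint for the Witch IceWater cosmetic back bling.
; On the post-res-in event it calls the parent, then attaches the P_Feather_A component to the skeletal mesh of a character part on the player pawn.
; The attachment snaps location and rotation to socket FX_Socket01, keeps relative scale and welds simulated bodies.
; The construction script is empty.
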(event EventOnPostResIn
  (|CallParentFunction)
  (Transformation|AttachComponentToComponent (Variables|Default|GetP_Feather_A) (CustomCharacter|GetSkeletalMeshForPartOnPlayerPawn) "FX_Socket01" "SnapToTarget" "SnapToTarget" "KeepRelative" true))

(fn ConstructionScript)
