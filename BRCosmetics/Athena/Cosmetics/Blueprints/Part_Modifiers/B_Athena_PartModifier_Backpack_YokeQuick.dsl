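; Part modifier for the YokeQuick back bling that drives reactive parameters on its idle particle effect.
; After res-in, a Parts-variant lookup for Stage1 toggles reactivity.
; In the frontend, an anim blueprint event sets ReactivityTime; otherwise, a watched kill stat raises it briefly, then resets it after a delay.
; Weapon changes trigger a delayed pickaxe meta-tag check that sets the effect's pickaxe-equipped flag.
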(event Custom|ReactiveFE
  (params (EnemyPlayerEliminated :type Boolean))
  (bind _returnvalue (CallFunction|GetIdleVFXComponent))
  (if (CustomCosmeticModifier|IsinFrontEnd)
    (if EnemyPlayerEliminated
      (Effects|Components|ParticleSystem|SetFloatParameter _returnvalue "ReactivityTime" 1.0)
      (else
        (Effects|Components|ParticleSystem|SetFloatParameter _returnvalue "ReactivityTime" 0.0)))))

(event EventOnPostResIn
  (bind _returnvalue (CallFunction|GetIdleVFXComponent))
  (bind _returnvalue_2 (CallFunction|GetIdleVFXComponent))
  (|CallParentFunction)
  (Variables|Default|SetSkeletalMesh (CustomCharacter|GetSkeletalMeshForPartOnPlayerPawn))
  (bind (_index _bindexisvalid) (CosmeticLoadout|GetVariantChannelActiveIndexforItemActor "/BRCosmetics/Athena/Items/Cosmetics/Backpacks/Backpack_YokeQuick.Backpack_YokeQuick" (CustomCharacter|GetAssociatedPlayerPawn) "(TagName=\"Cosmetics.Variant.Channel.Parts\")" (Utilities|Array|MakeArray "(TagName=\"Cosmetics.Variant.Property.Stage1\")")))
  (if _bindexisvalid
    (Effects|Components|ParticleSystem|SetBoolParameter _returnvalue_2 "bReactivityOn" true)
    (else
      (Effects|Components|ParticleSystem|SetBoolParameter _returnvalue_2 "bReactivityOn" false)))
  (if (Reactive|Frontend|IsCosmeticObjectDisplayedinFrontendContext 0)
    (bind _asskeletal_mesh_component (Utilities|Casting|CastToSkeletalMeshComponent (Variables|Default|GetSkeletalMesh)))
    (bind _asm_med_yoke_quick_pack_anim_bp (Utilities|Casting|CastToM_MED_YokeQuick_Pack_AnimBP (Components|SkeletalMesh|GetAnimInstance _asskeletal_mesh_component)))
    (Default|BindEventtoYokeQuickFETriggerFX _asm_med_yoke_quick_pack_anim_bp (delegate Custom|ReactiveFE))
    (else
      (bind (_statname _statvalue) (Cosmetics|WatchCosmeticStat (CustomCharacter|GetAssociatedPlayerPawn) "(TagName=\"StatManager.AthenaKills\")" -1 false)
        (:OnStatChanged
          (if (> _statvalue 0)
            (Effects|Components|ParticleSystem|SetFloatParameter _returnvalue "ReactivityTime" 1.0)
            (Utilities|FlowControl|Delay 4.5)
            (Effects|Components|ParticleSystem|SetFloatParameter _returnvalue "ReactivityTime" 0.0)
            (else
              (Effects|Components|ParticleSystem|SetFloatParameter _returnvalue "ReactivityTime" 0.0))))))))

(event EventOnWeaponChanged
  (params (NewWeapon :type Fort%20Weapon%20Object%20Reference) (OldWeapon :type Fort%20Weapon%20Object%20Reference))
  (bind _returnvalue (CallFunction|GetIdleVFXComponent))
  (|CallParentFunction NewWeapon OldWeapon)
  (Utilities|FlowControl|Delay 0.1)
  (bind _ascustom_character_part_anim_instance (Utilities|Casting|CastToCustomCharacterPartAnimInstance (Components|SkeletalMesh|GetAnimInstance (Customization|GetSkeletalMeshforPartType (CustomCharacter|GetAssociatedPlayerPawn) (CustomCharacter|GetCharacterPartType) :!class /Script/FortniteGame.FortPlayerPawn))))
  (if (CosmeticTags|PickaxeItemDefinitionHasAnyMetaTag _ascustom_character_part_anim_instance (Utilities|Array|MakeArray "(TagName=\"Cosmetics.MetaTags.YokeQuick\")") true)
    (Effects|Components|ParticleSystem|SetBoolParameter _returnvalue "bPickaxeEquipped" true)
    (else
      (Effects|Components|ParticleSystem|SetBoolParameter _returnvalue "bPickaxeEquipped" false))))

(fn ConstructionScript
  (|CallParentFunction))
