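; Part modifier for the York backpack cosmetic that reacts to the wearer's elimination count.
; After resources load, it watches the kills stat on the associated player pawn.
; Each change, after a 2.5 second delay, writes the count to the "Kills" scalar material parameter on the part's skeletal mesh.
; If the count exceeds the stored value, it plays an elimination sound attached to that mesh and saves the new value.
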(event EventOnPostResIn
  (|CallParentFunction)
  ;[1 1 1] Elimination
  (bind (_statname _statvalue) (Cosmetics|WatchCosmeticStat (CustomCharacter|GetAssociatedPlayerPawn) "(TagName=\"StatManager.AthenaKills\")")
    (:OnStatChanged
      (Utilities|FlowControl|Delay 2.5)
      (Rendering|Material|SetScalarParameterValueonMaterials (CustomCharacter|GetSkeletalMeshForPartOnPlayerPawn) "Kills" (Math|Conversions|ToFloat(Integer) _statvalue))
      (if (> _statvalue (Variables|Default|GetCurrent%20Stat))
        (CallFunction|PlayAudio)
        (Variables|Default|SetCurrent%20Stat _statvalue)))))

(event Custom|Play%20audio
  (Audio|SpawnSoundAttached (Variables|Default|GetElimination%20Sound) (CustomCharacter|GetSkeletalMeshForPartOnPlayerPawn)))

(fn ConstructionScript)
