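; Cosmetic part modifier for the BadBear character skin that plays landing effects when the player pawn lands.
; After calling the parent handler, it spawns the base and purple landing Niagara systems, both attached to the part skeletal mesh's root.
; Each successfully spawned system receives the pawn's character-slot loadout variants, so the effects follow the selected style.
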(event EventOnPawnLanded
  (bind _returnvalue (CustomCharacter|GetAssociatedPlayerPawn))
  (bind _returnvalue_2 (CustomCharacter|GetSkeletalMeshForPartOnPlayerPawn))
  (|CallParentFunction)
  (Utilities|FlowControl|Sequence
    (:then_0
      (bind _returnvalue_3 (PartModifier|SpawnSystem|SpawnSystemAttachedPartModifier "/BRCosmetics/Athena/Effects/Particles/Skins/BadBear/NS_Skin_BadBear_Landing.NS_Skin_BadBear_Landing" _returnvalue_2 "root")
        (:Success
          (Class|FortCosmeticFunctionLibrary|ApplyVariantstoComponentbySlotTag _returnvalue "(TagName=\"LoadoutSlot.Character\")" _returnvalue_3))))
    (:then_1
      (bind _returnvalue_4 (PartModifier|SpawnSystem|SpawnSystemAttachedPartModifier "/BRCosmetics/Athena/Effects/Particles/Skins/BadBear/NS_Skin_BadBear_Landing_purple.NS_Skin_BadBear_Landing_purple" _returnvalue_2 "root")
        (:Success
          (Class|FortCosmeticFunctionLibrary|ApplyVariantstoComponentbySlotTag _returnvalue "(TagName=\"LoadoutSlot.Character\")" _returnvalue_4))))))

(fn ConstructionScript
  (|CallParentFunction))
