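; BadCat backpack cosmetic part modifier that drives reactive idle VFX and material parameters on the part's skeletal mesh.
; After post-init in the frontend, it sets SprintReactive from a FlipFlop and passes game time and a frontend flag to materials.
; In game, it caches the owning pawn, and Tick feeds sliding or tactical sprint and jump states to the VFX.
; Normalized accumulated velocity also drives a material parameter.
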(event EventOnPostResIn
  (bind _returnvalue (CustomCharacter|GetSkeletalMeshForPartOnPlayerPawn))
  ;[0 0 0] post res
  (|CallParentFunction)
  (if (Reactive|Frontend|IsCosmeticObjectDisplayedinFrontendContext 0)
    (Utilities|FlowControl|Delay 1.0)
    (Utilities|FlowControl|Sequence
      (:then_0
        (bind _isa (Utilities|FlowControl|FlipFlop)
          (:A
            _)
          (:B
            _))
        (Effects|Components|ParticleSystem|SetBoolParameter (CallFunction|GetIdleVFXComponent) "SprintReactive" _isa)
        (Utilities|FlowControl|Delay 2.0))
      (:then_1
        (Rendering|Material|SetScalarParameterValueonMaterials _returnvalue "BP_GetTime" (Utilities|Time|GetGameTimeinSeconds))
        (Rendering|Material|SetScalarParameterValueonMaterials _returnvalue "BP_IsFrontEnd" 1.0)
        (Utilities|FlowControl|Delay 4.0)))
    (else
      (Rendering|Material|SetScalarParameterValueonMaterials (CustomCharacter|GetSkeletalMeshForPartOnPlayerPawn) "BP_IsFrontEnd" 0.0)
      (bind _asfort_player_pawn_athena (Utilities|Casting|CastToFortPlayerPawnAthena (CustomCharacter|GetAssociatedPlayerPawn)))
      (Variables|Default|SetplayerRef _asfort_player_pawn_athena))))

(event EventTick
  (params (DeltaSeconds :type Float%20%28single-precision%29))
  (bind _returnvalue (CallFunction|GetIdleVFXComponent))
  (bind _playerref (Variables|Default|GetplayerRef))
  ;[0 0 0] tick fx
  (|CallParentFunction DeltaSeconds)
  (if (not (Reactive|Frontend|IsCosmeticObjectDisplayedinFrontendContext 0))
    (Effects|Components|ParticleSystem|SetBoolParameter _returnvalue "SprintReactive" (or (Variables|/Script/FortniteGame.FortPawn|GetbIsSliding _playerref) (Variables|/Script/FortniteGame.FortPlayerPawn|GetbIsTacticalSprinting _playerref)))
    (Effects|Components|ParticleSystem|SetBoolParameter _returnvalue "PlayerJumping" (Variables|/Script/Engine.Character|GetbWasJumping _playerref))))

(event EventOnAccumulatedVelocity
  (params (NormalizedAccumulatedVelocity :type Float%20%28double-precision%29) (AccumulatedVelocity :type Float%20%28double-precision%29))
  (Rendering|Material|SetScalarParameterValueonMaterials (CustomCharacter|GetSkeletalMeshForPartOnPlayerPawn) "BP_AccumulatedVelocity" NormalizedAccumulatedVelocity))

(fn ConstructionScript
  (|CallParentFunction))
