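; Reactive part modifier for the Barium character cosmetic, driving materials and idle VFX from player movement and kills.
; Tick accumulates pawn speed with clamped, interpolated decay into paint material and lightning parameters on tagged part meshes.
; Rising kill stats timestamp a material boost, spawn variant-matched attached FX and play audio.
; A looping frontend timeline simulates these reactions, the Parts variant sets the VFX helmet flag, and vehicles disable idle VFX.
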(event EventOnPostResIn
  ;[1 1 1] Post Res In
  (|CallParentFunction)
  (Variables|Default|SetPlayerPawn (CustomCharacter|GetAssociatedPlayerPawn))
  (bind _index (Class|FortCosmeticFunctionLibrary|GetVariantChannelActiveIndexForSlot "(TagName=\"LoadoutSlot.Character\")" (CosmeticLoadout|GetCosmeticLoadout) "(TagName=\"Cosmetics.Variant.Channel.Parts\")" (Utilities|Array|MakeArray "(TagName=\"Cosmetics.Variant.Property.Stage3\")" "(TagName=\"Cosmetics.Variant.Property.Stage1\")" "(TagName=\"Cosmetics.Variant.Property.Stage2\")")))
  (bind _output_get (Variables|Default|SetHelmetBool (Utilities|Select :Index _index :Option%200 true :Option%201 false :Option%202 false)))
  (Effects|Components|ParticleSystem|SetBoolParameter (CallFunction|GetIdleVFXComponent) "bisHelmet" _output_get)
  (bind _output_get_2 (Variables|Default|SetIn%20Frontend (Reactive|Frontend|IsCosmeticObjectDisplayedinFrontendContext 0)))
  (if _output_get_2
    (bind (_direction _velocity _emissive) (Timeline|Barium_Frontend_Timline PlayFromStart)
      (:Update
        (Rendering|Material|SetScalarParameterValueonMaterials (Variables|Components|GetMesh%20Components) "PaintVelocity" (* _velocity (Variables|Default|GetMaxVelocityAccum)))
        (Rendering|Material|SetScalarParameterValueonMaterials (Variables|Components|GetMesh%20Components) "FE_EmissiveBoost" (* _emissive (Variables|Default|GetFE_EmissiveBoost))))
      (:Finished
        (CallFunction|FEBariumLoop))
      (:KillEvent
        (CallFunction|KillEventFX)))
    (else
      (Actor|Tick|SetActorTickEnabled true))))

(event Custom|FE_BariumLoop
  ;[1 1 1] Post Res In
  (Timeline|Barium_Frontend_Timline PlayFromStart))

(event EventOnStatChanged
  (params (State%20Name :type Name :default "None") (Stat%20Value :type Integer))
  ;[1 1 1] On Kill
  (|CallParentFunction State%20Name Stat%20Value)
  (if (> Stat%20Value (Variables|Default|GetKills))
    (Variables|Default|SetKills Stat%20Value)
    (Rendering|Material|SetScalarParameterValueonMaterials (Variables|Components|GetMesh%20Components) "KillBoostTime" (Utilities|Time|GetGameTimeinSeconds))
    (CallFunction|KillEventFX)
    (CallFunction|TriggerAudio)))

(event Custom|KillEventFX
  (bind _killfxsystem (Variables|Default|GetKillFXSystem))
  ;[1 1 1] Also set value on vfx component
  (if (not (Variables|/Script/FortniteGame.FortPlayerPawn|GetbIsInFrontEndHologram (CustomCharacter|GetAssociatedPlayerPawn)))
    (Utilities|IsValid _killfxsystem
      (:Is%20Valid
        (bind _returnvalue (PartModifier|SpawnSystem|SpawnSystemAttachedPartModifier _killfxsystem (CustomCharacter|GetSkeletalMeshForPartOnPlayerPawn))
          (:Success
            (Class|FortCosmeticFunctionLibrary|ApplyVariantstoComponentbySlotTag (CustomCharacter|GetAssociatedPlayerPawn) "(TagName=\"LoadoutSlot.Character\")" _returnvalue)))))))

(event EventTick
  (params (DeltaSeconds :type Float%20%28single-precision%29))
  (bind _returnvalue (CallFunction|GetIdleVFXComponent))
  (bind _maxvelocityaccum (Variables|Default|GetMaxVelocityAccum))
  ;[1 1 1] Velocity Material and FX Params
  (if (not (Variables|Default|GetIn%20Frontend))
    (bind _output_get (Variables|Default|SetVelocityAccum (+ (Math|Vector|VectorLength (Transformation|GetVelocity (CustomCharacter|GetAssociatedPlayerPawn))) (Variables|Default|GetVelocityAccum))))
    (Utilities|FlowControl|Sequence
      (:then_0
        (Rendering|Material|SetScalarParameterValueonMaterials (Variables|Components|GetMesh%20Components) "PaintVelocity" (Math|Interpolation|FInterpTo (Math|Float|Clamp(Float) _output_get 0.0 _maxvelocityaccum) 0.0 DeltaSeconds 0.1))
        (Variables|Default|SetVelocityAccum (Math|Interpolation|FInterpTo (Math|Float|Clamp(Float) _output_get 0.0 _maxvelocityaccum) 0.0 DeltaSeconds 0.1))
        (Utilities|IsSkydiving (Variables|Default|GetPlayerPawn)
          (:true
            (Rendering|Material|SetScalarParameterValueonMaterials (Variables|Components|GetMesh%20Components) "PaintVelocityClamp" 1.0))
          (:false
            (Rendering|Material|SetScalarParameterValueonMaterials (Variables|Components|GetMesh%20Components) "PaintVelocityClamp" 0.6))))
      (:then_1
        ;;[1 1 1] Also set values on vfx component
        (Effects|Components|ParticleSystem|SetFloatParameter _returnvalue "Distance" _maxvelocityaccum)
        (Effects|Components|ParticleSystem|SetFloatParameter _returnvalue "LightningAcceleration" (Math|Interpolation|FInterpTo (Math|Float|Clamp(Float) _output_get 0.0 _maxvelocityaccum) 0.0 DeltaSeconds 0.1))))))

(event EventOnCharacterCustomizationFinished
  (params (CharacterPartInfo :type Array%20of%20Applied%20Character%20Part%20Info%20Structures))
  ;[1 1 1] On Character Customization Finished
  (bind _returnvalue (Class|FortCosmeticFunctionLibrary|FilterPartMeshComponentsbyTag CharacterPartInfo "(TagName=\"CharacterPart.Character.Barium.Reactive\")"))
  (bind _output_get (Variables|Components|SetMesh%20Components _returnvalue))
  (Rendering|Material|SetScalarParameterValueonMaterials _output_get "PaintVelocityMax" (Variables|Default|GetMaxVelocityAccum)))

(event EventOnEnterVehicle
  (params (VehicleEntered :type Fort%20Vehicle%20Interface%20Interface%20%28by%20ref%29))
  (Components|Activation|Deactivate (CallFunction|GetIdleVFXComponent)))

(event EventOnExitVehicle
  (Components|Activation|Activate (CallFunction|GetIdleVFXComponent) false))

(event Custom|TriggerAudio
  ;[1 1 1] Audio
  (Audio|SpawnSoundAttached (Variables|Default|GetSound) (CustomCharacter|GetSkeletalMeshForPartOnPlayerPawn)))

(fn ConstructionScript
  (|CallParentFunction))
