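; Cosmetic part modifier Blueprint for the BariumDemon character body that adjusts how its idle VFX renders.
; On the post-res-in event it calls the parent logic, then, if the idle VFX component is valid, sets its translucent sort priority negative so it sorts behind nearly everything.
; The construction script only calls the parent.
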(event EventOnPostResIn
  (bind _returnvalue (CallFunction|GetIdleVFXComponent))
  (|CallParentFunction)
  (Utilities|IsValid _returnvalue
    (:Is%20Valid
      ;[1 1 1] This idle component should sort behind nearly everything
      (Rendering|SetTranslucentSortPriority _returnvalue -1))))

(fn ConstructionScript
  (|CallParentFunction))
