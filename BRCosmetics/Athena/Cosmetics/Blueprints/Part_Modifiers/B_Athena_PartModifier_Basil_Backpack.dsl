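; Cosmetic part modifier for the Basil backpack that makes its idle VFX react to player kills.
; On post-res-in it watches the Athena kills stat and enables reactivity from the loadout's Parts variant check against Stage1.
; When enabled, each kill plays a timeline driving particle Intensity, ending at a fixed value.
; In frontend previews, a separate timeline drives Intensity and toggles activeKill on the pack's anim blueprint.
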(event EventOnPostResIn
  (|CallParentFunction)
  (Utilities|FlowControl|Sequence
    (:then_0
      (if (Reactive|Frontend|ShouldDisplayFrontendReactivity 0 "Backpack")
        (bind _asm_med_basil_pack_anim_bp (Utilities|Casting|CastToM_MED_Basil_Pack_AnimBP (Components|SkeletalMesh|GetAnimInstance (Customization|GetSkeletalMeshforPartType (CustomCharacter|GetAssociatedPlayerPawn) (CustomCharacter|GetCharacterPartType) :!class /Script/FortniteGame.FortPlayerPawn))))
        (Variables|Default|SetAnim%20BP _asm_med_basil_pack_anim_bp)
        (bind (_direction _fxintensity) (Timeline|Frontend_ReactivityLoop Play)
          (:Update
            (Effects|Components|ParticleSystem|SetFloatParameter (CallFunction|GetIdleVFXComponent) "Intensity" _fxintensity))
          (:Finished
            (Variables|/BRCosmetics/Accessories/FORT_Backpacks/Backpack_M_MED_Basil/Meshes/M_MED_Basil_Pack_AnimBP.M_MED_Basil_Pack_AnimBP_C|SetactiveKill false (Variables|Default|GetAnim%20BP)))
          (:Activate%20Front%20End%20Reactivity
            (Variables|/BRCosmetics/Accessories/FORT_Backpacks/Backpack_M_MED_Basil/Meshes/M_MED_Basil_Pack_AnimBP.M_MED_Basil_Pack_AnimBP_C|SetactiveKill true (Variables|Default|GetAnim%20BP))))))
    (:then_1
      ;[1 1 1] Setup On Kill
      (Cosmetics|WatchCosmeticStat (CustomCharacter|GetAssociatedPlayerPawn) "(TagName=\"StatManager.AthenaKills\")" 0 false
        (:OnStatChanged
          (CallFunction|OnKill))))
    (:then_2
      ;[1 1 1] Read CIDs
      (bind (_index _bindexisvalid) (Class|FortCosmeticFunctionLibrary|GetVariantChannelActiveIndexForItem "/BRCosmetics/Athena/Items/Cosmetics/Backpacks/Backpack_Basil.Backpack_Basil" (CosmeticLoadout|GetCosmeticLoadout) "(TagName=\"Cosmetics.Variant.Channel.Parts\")" (Utilities|Array|MakeArray "(TagName=\"Cosmetics.Variant.Property.Stage1\")") false false))
      (Variables|Default|SetReactivity%20Enabled _bindexisvalid))))

(event Custom|OnKill
  (bind _returnvalue (CallFunction|GetIdleVFXComponent))
  ;[1 1 1] On Kill Event
  (if (Variables|Default|GetReactivity%20Enabled)
    (bind (_direction _intensity) (Timeline|FX_Intensity Play)
      (:Update
        (Effects|Components|ParticleSystem|SetFloatParameter _returnvalue "Intensity" _intensity))
      (:Finished
        (Effects|Components|ParticleSystem|SetFloatParameter _returnvalue "Intensity" 1.0)))))

(fn ConstructionScript)
