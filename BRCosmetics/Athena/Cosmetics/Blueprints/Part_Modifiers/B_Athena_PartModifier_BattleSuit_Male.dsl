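; Character part modifier for the male BattleSuit cosmetic that shows or hides foot effects when the player's weapon changes.
; On weapon change it calls the parent, then runs a hidden-part check on the idle VFX component using the HideFX.Feet gameplay tag.
; The inverted result becomes a float driving the MasterAlpha material parameter on the part's skeletal mesh, zeroing it while hidden.
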(event EventOnWeaponChanged
  (params (NewWeapon :type Fort%20Weapon%20Object%20Reference) (OldWeapon :type Fort%20Weapon%20Object%20Reference))
  (|CallParentFunction NewWeapon OldWeapon)
  (bind _isparthidden (Default|HiddenPartCheck-Weapon (Utilities|Array|MakeArray (CallFunction|GetIdleVFXComponent)) NewWeapon OldWeapon true "(TagName=\"Animation.CharacterParts.HideFX.Feet\")" "bisFootHidden"))
  (Rendering|Material|SetScalarParameterValueonMaterials (CustomCharacter|GetSkeletalMeshForPartOnPlayerPawn) "MasterAlpha" (Math|Conversions|ToFloat(Boolean) (not _isparthidden))))

(fn ConstructionScript
  (|CallParentFunction))
