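; Battlehound cosmetic part modifier Blueprint that adjusts the cosmetic's idle particle effect around skydiving.
; On skydive begin and end, it calls the parent handler, then sets the SpawnPerUnitRate float parameter on the idle VFX component.
; The parameter drops to zero during skydiving and returns to one afterward, likely suppressing distance-based particle spawning while falling.
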(event EventOnBeginSkydiving
  (|CallParentFunction)
  (Effects|Components|ParticleSystem|SetFloatParameter (CallFunction|GetIdleVFXComponent) "SpawnPerUnitRate" 0.0))

(event EventOnEndSkydiving
  (|CallParentFunction)
  (Effects|Components|ParticleSystem|SetFloatParameter (CallFunction|GetIdleVFXComponent) "SpawnPerUnitRate" 1.0))

(fn ConstructionScript)
