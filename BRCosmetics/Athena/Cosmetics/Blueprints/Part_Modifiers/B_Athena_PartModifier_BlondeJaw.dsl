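; Cosmetic part modifier for BlondeJaw that makes head and body materials and idle VFX react to damage dealt and eliminations.
; In-match stat watchers fire the reactions, while the front end replays a timed burst of damage reactions on a loop.
; Each reaction stamps game time into round-robin parameters (five damage slots, two elimination), once per frame at most.
; After spawn it also enables Niagara back FX when no real backpack is equipped.
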(event EventOnPostResIn
  (bind _returnvalue (CustomCharacter|GetAssociatedPlayerPawn))
  (bind _returnvalue_2 (CallFunction|GetIdleVFXComponent))
  (|CallParentFunction)
  (Utilities|FlowControl|Sequence
    (:then_0
      (Utilities|IsValid _returnvalue_2
        (:Is%20Valid
          (bind _asniagara_particle_system_component (Utilities|Casting|CastToNiagaraComponent _returnvalue_2))
          (Niagara|SetNiagaraVariable(Bool) _asniagara_particle_system_component "bShowBackFX" (or (not (Utilities|IsValid (CosmeticLoadout|GetItemDefinitionfromLoadout (CosmeticLoadout|GetCosmeticLoadout) "(TagName=\"LoadoutSlot.Backpack\")") :!class /Script/Engine.KismetSystemLibrary)) (== (CosmeticLoadout|GetItemDefinitionfromLoadout (CosmeticLoadout|GetCosmeticLoadout) "(TagName=\"LoadoutSlot.Backpack\")") "/Game/Athena/Items/Cosmetics/Backpacks/BID_STWHeroNoDefaultBackpack.BID_STWHeroNoDefaultBackpack"))))))
    (:then_1
      (if (CustomCosmeticModifier|IsinFrontEnd)
        (Utilities|FlowControl|Delay 4.5)
        (CallFunction|FrontendReacitvityLoop)
        (Utilities|Time|SetTimerbyEvent (delegate Custom|FrontendReacitvityLoop) 11.0 true)
        (else
          (Utilities|FlowControl|Sequence
            (:then_0
              (Cosmetics|WatchCosmeticStat _returnvalue "(TagName=\"StatManager.DamageDealt\")" 0 false
                (:OnStatChanged
                  (CallFunction|TriggerDamageReactivity))))
            (:then_1
              (Cosmetics|WatchCosmeticStat _returnvalue "(TagName=\"StatManager.AthenaKills\")" 0 false
                (:OnStatChanged
                  (CallFunction|TrigerElimReactivity))))))))))

(event Custom|FrontendReacitvityLoop
  (bind _returnvalue 0.4)
  (CallFunction|TriggerDamageReactivity)
  (Utilities|FlowControl|Delay _returnvalue)
  (CallFunction|TriggerDamageReactivity)
  (Utilities|FlowControl|Delay _returnvalue)
  (CallFunction|TriggerDamageReactivity)
  (Utilities|FlowControl|Delay _returnvalue)
  (CallFunction|TriggerDamageReactivity)
  (Utilities|FlowControl|Delay _returnvalue)
  (CallFunction|TriggerDamageReactivity))

(event Custom|TriggerDamageReactivity
  (bind _dmgreactivitycounter (Variables|Default|GetDmgReactivityCounter))
  (bind _returnvalue (Math|Integer|%25%28Integer%29 _dmgreactivitycounter 5))
  (bind _returnvalue_2 (Utilities|String|ToString(Integer) _returnvalue))
  (bind _returnvalue_3 (Utilities|String|Append "DmgReactivityStartTime_" _returnvalue_2))
  (bind _returnvalue_4 (Utilities|String|StringToName _returnvalue_3))
  (Utilities|DoOncePerFrame
    (:%20%20
      (Math|Integer|IncrementInt (Variables|Default|GetDmgReactivityCounter)
        (:%20%20
          (Rendering|Material|SetScalarParameterValueonMaterials (CallFunction|GetSkelMeshCompstoModify) _returnvalue_4 (Utilities|Time|GetGameTimeinSeconds))
          (Effects|Components|ParticleSystem|SetFloatParameter (CallFunction|GetIdleVFXComponent) _returnvalue_4 (Utilities|Time|GetGameTimeinSeconds)))))))

(event Custom|TrigerElimReactivity
  (bind _elimreactivitycounter (Variables|Default|GetElimReactivityCounter))
  (bind _returnvalue (Math|Integer|%25%28Integer%29 _elimreactivitycounter 2))
  (bind _returnvalue_2 (Utilities|String|ToString(Integer) _returnvalue))
  (bind _returnvalue_3 (Utilities|String|Append "ElimReactivityStartTime_" _returnvalue_2))
  (bind _returnvalue_4 (Utilities|String|StringToName _returnvalue_3))
  (Utilities|DoOncePerFrame
    (:%20%20
      (Math|Integer|IncrementInt (Variables|Default|GetElimReactivityCounter)
        (:%20%20
          (Rendering|Material|SetScalarParameterValueonMaterials (CallFunction|GetSkelMeshCompstoModify) _returnvalue_4 (Utilities|Time|GetGameTimeinSeconds))
          (Effects|Components|ParticleSystem|SetFloatParameter (CallFunction|GetIdleVFXComponent) _returnvalue_4 (Utilities|Time|GetGameTimeinSeconds)))))))

(fn GetSkelMeshCompsToModify
  (return (Utilities|Array|MakeArray (Customization|GetSkeletalMeshforPartType (CustomCharacter|GetAssociatedPlayerPawn) "Head" :!class /Script/FortniteGame.FortPlayerPawn) (Customization|GetSkeletalMeshforPartType (CustomCharacter|GetAssociatedPlayerPawn) "Body" :!class /Script/FortniteGame.FortPlayerPawn))))

(fn ConstructionScript
  (|CallParentFunction))
