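; Blueprint part modifier for a cosmetic named BlondeJaw, labeled as a universal backpack variant.
; It overrides GetSkelMeshCompsToModify to return a one-element array holding the skeletal mesh from GetSkeletalMeshForPartOnPlayerPawn, with the part argument left at its default.
; The construction script only calls the parent, so the modification itself is not defined in this graph.
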(fn GetSkelMeshCompsToModify
  (return (Utilities|Array|MakeArray (CustomCharacter|GetSkeletalMeshForPartOnPlayerPawn))))

(fn ConstructionScript
  (|CallParentFunction))
