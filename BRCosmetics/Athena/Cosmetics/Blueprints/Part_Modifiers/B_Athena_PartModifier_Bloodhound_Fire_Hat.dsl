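; Cosmetic part modifier for the Bloodhound Fire hat that manages visibility of its idle VFX when the character's head is hidden.
; After resurrection and on weapon changes, it calls the parent handler, then passes the idle VFX component to a shared weapon-based hidden-part check.
; That check uses the HideFX.Head gameplay tag and the "bisHeadHidden" name, with empty weapon inputs in the resurrection case.
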(event EventOnPostResIn
  (|CallParentFunction)
  (Default|HiddenPartCheck-Weapon (Utilities|Array|MakeArray (CallFunction|GetIdleVFXComponent)) 0 0 false "(TagName=\"Animation.CharacterParts.HideFX.Head\")" "bisHeadHidden"))

(event EventOnWeaponChanged
  (params (NewWeapon :type Fort%20Weapon%20Object%20Reference) (OldWeapon :type Fort%20Weapon%20Object%20Reference))
  (|CallParentFunction NewWeapon OldWeapon)
  (Default|HiddenPartCheck-Weapon (Utilities|Array|MakeArray (CallFunction|GetIdleVFXComponent)) NewWeapon OldWeapon false "(TagName=\"Animation.CharacterParts.HideFX.Head\")" "bisHeadHidden"))

(fn ConstructionScript)
