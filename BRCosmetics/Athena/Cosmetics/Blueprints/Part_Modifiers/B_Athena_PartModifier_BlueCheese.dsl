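; Part modifier for the Blue Cheese outfit that reactively steps its body through three Shock material stages.
; In frontend display, a timeline cycles the stages; cosmetic state activations and part changes trigger a delayed update to the last active stage.
; Updates and frontend display also spawn a Niagara burst flagged for frontend.
; Material swaps preserve the chrome strength parameter, and the first and final stages toggle FinalStage on the idle VFX.
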(event EventOnPostResIn
  (|CallParentFunction)
  (CallFunction|GetParts)
  (Utilities|FlowControl|Sequence
    (:then_0
      (if (Reactive|Frontend|ShouldDisplayFrontendReactivity 0 "Backpack")
        (Timeline|Stage%20Cycle%20Transition PlayFromStart
          (:Stage%201
            (CallFunction|StageOne))
          (:Stage%202
            (CallFunction|StageTwo))
          (:Stage%203
            (CallFunction|StageThree)))
        (bind _returnvalue (PartModifier|SpawnSystem|SpawnSystemAttachedPartModifier "/BRCosmetics/Characters/Player/Male/Medium/Bodies/M_MED_Blue_Cheese/FX/NS_Blue_Cheese_M_OnBody_Burst.NS_Blue_Cheese_M_OnBody_Burst" (CustomCharacter|GetSkeletalMeshForPartOnPlayerPawn) "root")
          (:Success
            (Niagara|SetNiagaraVariable(Bool) _returnvalue "IsFrontEnd" (Reactive|Frontend|ShouldDisplayFrontendReactivity 0 "Backpack"))))))
    (:then_1
      (EventDispatchers|BindEventtoOnPartChanged (CustomCharacter|GetAssociatedPlayerPawn) (delegate Custom|Part%20Changed)))))

(event Custom|Part%20Changed
  (params (PartType :type EFortCustomPartType%20Enum :default "Head"))
  (CallFunction|UpdateFX))

(event Custom|Update%20FX
  (Utilities|FlowControl|Delay 0.3)
  (switch int (Variables|Default|GetLast%20Active%20Stage)
    (:0
      (CallFunction|StageOne))
    (:1
      (CallFunction|StageTwo))
    (:2
      (CallFunction|StageThree)))
  (bind _returnvalue (PartModifier|SpawnSystem|SpawnSystemAttachedPartModifier "/BRCosmetics/Characters/Player/Male/Medium/Bodies/M_MED_Blue_Cheese/FX/NS_Blue_Cheese_M_OnBody_Burst.NS_Blue_Cheese_M_OnBody_Burst" (CustomCharacter|GetSkeletalMeshForPartOnPlayerPawn) "root")
    (:Success
      (Niagara|SetNiagaraVariable(Bool) _returnvalue "IsFrontEnd" (Reactive|Frontend|ShouldDisplayFrontendReactivity 0 "Backpack")))))

(event EventOnCosmeticStateActivated
  (params (Stat%20Name :type Name :default "None") (Stat%20Value :type Integer) (STAGE%20Num%20Being%20Activated :type Integer))
  (|CallParentFunction Stat%20Name Stat%20Value STAGE%20Num%20Being%20Activated)
  (Variables|Default|SetLast%20Active%20Stage STAGE%20Num%20Being%20Activated)
  (CallFunction|UpdateFX))

(event Custom|StageOne
  (bind _body (Variables|Default|Getbody))
  (Effects|Components|ParticleSystem|SetFloatParameter (CallFunction|GetIdleVFXComponent) "FinalStage" 0.0)
  ;[1 1 1] modified for chrome : gets the value of S22ChromeCharacters_strength before and after material swap
  (bind _returnvalue (Fortnite|GetDynamicMaterialInstance _body))
  (bind _returnvalue_2 (Rendering|Material|GetScalarParameterValue _returnvalue "S22ChromeCharacters_strength" :!class /Script/Engine.MaterialInstanceDynamic))
  (Rendering|Material|SetMaterial _body 0 "/BRCosmetics/Characters/Player/Male/Medium/Bodies/M_MED_Blue_Cheese/Materials/MI_MED_Blue_Cheese_Shock_T0.MI_MED_Blue_Cheese_Shock_T0")
  (Rendering|Material|SetScalarParameterValueonMaterials _body "S22ChromeCharacters_strength" _returnvalue_2))

(event Custom|StageTwo
  (bind _body (Variables|Default|Getbody))
  ;[1 1 1] modified for chrome : gets the value of S22ChromeCharacters_strength before and after material swap
  (bind _returnvalue (Fortnite|GetDynamicMaterialInstance _body))
  (bind _returnvalue_2 (Rendering|Material|GetScalarParameterValue _returnvalue "S22ChromeCharacters_strength" :!class /Script/Engine.MaterialInstanceDynamic))
  (Rendering|Material|SetMaterial _body 0 "/BRCosmetics/Characters/Player/Male/Medium/Bodies/M_MED_Blue_Cheese/Materials/MI_MED_Blue_Cheese_Shock_T1.MI_MED_Blue_Cheese_Shock_T1")
  (Rendering|Material|SetScalarParameterValueonMaterials _body "S22ChromeCharacters_strength" _returnvalue_2))

(event Custom|StageThree
  (bind _body (Variables|Default|Getbody))
  (Effects|Components|ParticleSystem|SetFloatParameter (CallFunction|GetIdleVFXComponent) "FinalStage" 1.0)
  ;[1 1 1] modified for chrome : gets the value of S22ChromeCharacters_strength before and after material swap
  (bind _returnvalue (Fortnite|GetDynamicMaterialInstance _body))
  (bind _returnvalue_2 (Rendering|Material|GetScalarParameterValue _returnvalue "S22ChromeCharacters_strength" :!class /Script/Engine.MaterialInstanceDynamic))
  (Rendering|Material|SetMaterial _body 0 "/BRCosmetics/Characters/Player/Male/Medium/Bodies/M_MED_Blue_Cheese/Materials/MI_MED_Blue_Cheese_Shock_T2.MI_MED_Blue_Cheese_Shock_T2")
  (Rendering|Material|SetScalarParameterValueonMaterials _body "S22ChromeCharacters_strength" _returnvalue_2))

(event Custom|StageFour)

(fn GetParts
  (bind _returnvalue (CustomCharacter|GetAssociatedPlayerPawn))
  (Variables|Default|Sethead (Customization|GetSkeletalMeshforPartType _returnvalue "Head" :!class /Script/FortniteGame.FortPlayerPawn))
  (Variables|Default|Setbody (Customization|GetSkeletalMeshforPartType _returnvalue "Body" :!class /Script/FortniteGame.FortPlayerPawn))
  (Variables|Default|Setface (Customization|GetSkeletalMeshforPartType _returnvalue "Face" :!class /Script/FortniteGame.FortPlayerPawn)))

(fn SetScalarParameterOnMaterialsMultiParts
  (params (name :type Name) (value :type Float%20%28double-precision%29))
  (Rendering|Material|SetScalarParameterValueonMaterials (Variables|Default|Gethead) name value)
  (Rendering|Material|SetScalarParameterValueonMaterials (Variables|Default|Getface) name value)
  (Rendering|Material|SetScalarParameterValueonMaterials (Variables|Default|Getbody) name value))

(fn SetVectorParameteronMaterialsMultiParts
  (params (name :type Name) (value :type Vector))
  (Rendering|Material|SetVectorParameterValueonMaterials (Variables|Default|Gethead) name value)
  (Rendering|Material|SetVectorParameterValueonMaterials (Variables|Default|Getface) name value)
  (Rendering|Material|SetVectorParameterValueonMaterials (Variables|Default|Getbody) name value))

(fn ConstructionScript
  (|CallParentFunction))
