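; Character part modifier for a female Blue Flames cosmetic that manages flame effects on the character's hands.
; After the parent's post-resource-init logic, it snaps the other-hand FX to the middle_02_l socket on the character part's skeletal mesh.
; On weapon change, it passes that FX and the idle VFX to a hidden-part check keyed on the HideFX.Hands gameplay tag.
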(event EventOnPostResIn
  (|CallParentFunction)
  (Transformation|AttachComponentToComponent (Variables|Default|GetOtherHandFX) (CustomCharacter|GetSkeletalMeshForPartOnPlayerPawn) "middle_02_l" "SnapToTarget" "SnapToTarget" "KeepRelative" true))

(event EventOnWeaponChanged
  (params (NewWeapon :type Fort%20Weapon%20Object%20Reference) (OldWeapon :type Fort%20Weapon%20Object%20Reference))
  (|CallParentFunction NewWeapon OldWeapon)
  (Default|HiddenPartCheck-Weapon (Utilities|Array|MakeArray (Variables|Default|GetOtherHandFX) (CallFunction|GetIdleVFXComponent)) NewWeapon OldWeapon false "(TagName=\"Animation.CharacterParts.HideFX.Hands\")" "bisHandHidden"))

(fn ConstructionScript)
