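; Cosmetic part modifier that drives a character body part's materials from an accumulating kill count.
; After post-resurrection-in, it sets offset and front-end flag parameters, adding a preset kill count and game-time stamp in the front end.
; Each stat change stores the value as the kill count and pushes it with game time to the materials.
; Counts above ten replay a timeline that animates BlendTime and ElimCurve.
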(event EventOnPostResIn
  (bind _returnvalue (CustomCharacter|GetSkeletalMeshForPartOnPlayerPawn))
  (|CallParentFunction)
  (if (CustomCosmeticModifier|IsinFrontEnd)
    (Rendering|Material|SetScalarParameterValueonMaterials _returnvalue "Offset Amount (S)" 24.0)
    (Rendering|Material|SetScalarParameterValueonMaterials _returnvalue "HC_KillCount" 1e+01)
    (Rendering|Material|SetScalarParameterValueonMaterials _returnvalue "IsInFrontEnd" 1.0)
    (Rendering|Material|SetScalarParameterValueonMaterials _returnvalue "HC_FEGameTimeFromBP" (Utilities|Time|GetGameTimeinSeconds))
    (else
      (Rendering|Material|SetScalarParameterValueonMaterials _returnvalue "Offset Amount (S)" 21.0)
      (Rendering|Material|SetScalarParameterValueonMaterials _returnvalue "IsInFrontEnd" 0.0))))

(event EventOnStatChanged
  (params (State%20Name :type Name :default "None") (Stat%20Value :type Integer))
  (bind _returnvalue (CustomCharacter|GetSkeletalMeshForPartOnPlayerPawn))
  (bind _returnvalue_2 (CustomCharacter|GetSkeletalMeshForPartOnPlayerPawn))
  (bind _kill_count (Variables|Default|GetKill%20Count))
  (|CallParentFunction State%20Name Stat%20Value)
  (Variables|Default|SetKill%20Count (Math|Conversions|ToFloat(Integer) Stat%20Value))
  (Rendering|Material|SetScalarParameterValueonMaterials _returnvalue_2 "HC_KillCount" _kill_count)
  (Rendering|Material|SetScalarParameterValueonMaterials _returnvalue_2 "HC_GameTimeFromBP" (Utilities|Time|GetGameTimeinSeconds))
  (if (> _kill_count 1e+01)
    (bind (_direction _blendtime _elimcurve) (Timeline|Timeline PlayFromStart)
      (:Update
        (Rendering|Material|SetScalarParameterValueonMaterials _returnvalue "BlendTime" _blendtime)
        (Rendering|Material|SetScalarParameterValueonMaterials _returnvalue "ElimCurve" _elimcurve)))))

(fn ConstructionScript
  (|CallParentFunction))
