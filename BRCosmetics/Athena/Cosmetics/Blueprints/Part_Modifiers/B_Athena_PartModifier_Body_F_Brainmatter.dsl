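; Body part modifier for the female Brainmatter character cosmetic that matches its idle VFX to the equipped style variants.
; After calling the parent on post res-in, it reads the active Material and Parts variant indices from the pawn's character loadout slot.
; Nested branches set three particle bool parameters (glow, glow jacket, default jacket), enabling at most one and none when neither index is 1.
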(event EventOnPostResIn
  (bind _returnvalue (CustomCharacter|GetAssociatedPlayerPawn))
  (bind _returnvalue_2 (CallFunction|GetIdleVFXComponent))
  (bind _channelparts (Variables|Default|GetChannelParts))
  (bind _variantliststage (Variables|Default|GetVariantListStage))
  (|CallParentFunction)
  (bind _index (CosmeticLoadout|GetVariantChannelActiveIndexforSlotActor "(TagName=\"LoadoutSlot.Character\")" _returnvalue (Variables|Default|GetChannel%20Material) (Variables|Default|GetVariantListMaterial)))
  (if (== _index 1)
    (bind _index_2 (CosmeticLoadout|GetVariantChannelActiveIndexforSlotActor "(TagName=\"LoadoutSlot.Character\")" _returnvalue _channelparts _variantliststage))
    (if (== _index_2 1)
      (Effects|Components|ParticleSystem|SetBoolParameter _returnvalue_2 "GlowVariantActive" false)
      (Effects|Components|ParticleSystem|SetBoolParameter _returnvalue_2 "GlowJacketVariantActive" true)
      (Effects|Components|ParticleSystem|SetBoolParameter _returnvalue_2 "DefaultJacketVariantActive" false)
      (else
        (Effects|Components|ParticleSystem|SetBoolParameter _returnvalue_2 "GlowVariantActive" true)
        (Effects|Components|ParticleSystem|SetBoolParameter _returnvalue_2 "GlowJacketVariantActive" false)
        (Effects|Components|ParticleSystem|SetBoolParameter _returnvalue_2 "DefaultJacketVariantActive" false)))
    (else
      (bind _index_3 (CosmeticLoadout|GetVariantChannelActiveIndexforSlotActor "(TagName=\"LoadoutSlot.Character\")" _returnvalue _channelparts _variantliststage))
      (if (== _index_3 1)
        (Effects|Components|ParticleSystem|SetBoolParameter _returnvalue_2 "GlowVariantActive" false)
        (Effects|Components|ParticleSystem|SetBoolParameter _returnvalue_2 "GlowJacketVariantActive" false)
        (Effects|Components|ParticleSystem|SetBoolParameter _returnvalue_2 "DefaultJacketVariantActive" true)
        (else
          (Effects|Components|ParticleSystem|SetBoolParameter _returnvalue_2 "GlowVariantActive" false)
          (Effects|Components|ParticleSystem|SetBoolParameter _returnvalue_2 "GlowJacketVariantActive" false)
          (Effects|Components|ParticleSystem|SetBoolParameter _returnvalue_2 "DefaultJacketVariantActive" false))))))

(fn ConstructionScript
  (|CallParentFunction))
