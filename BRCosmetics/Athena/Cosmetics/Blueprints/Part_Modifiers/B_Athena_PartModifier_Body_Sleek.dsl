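; Body part modifier for the Sleek character cosmetic that adds damage-reactive VFX when the Stage3 Parts variant check passes.
; On post-resurrection init (EventOnPostResIn), it watches the owning pawn's DamageDealt stat.
; Positive values spawn an attached flower-hit Niagara effect, gated by a DoOnce, retriggerable delay and ResetFX loop.
; At a damage threshold, it also activates the idle VFX component as an aura.
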(event EventOnPostResIn
  (|CallParentFunction)
  (bind (_index _bindexisvalid) (Class|FortCosmeticFunctionLibrary|GetVariantChannelActiveIndexForSlot "(TagName=\"LoadoutSlot.Character\")" (CosmeticLoadout|GetCosmeticLoadout) "(TagName=\"Cosmetics.Variant.Channel.Parts\")" (Utilities|Array|MakeArray "(TagName=\"Cosmetics.Variant.Property.Stage3\")") true false))
  (if _bindexisvalid
    (bind (_statname _statvalue) (Cosmetics|WatchCosmeticStat (CustomCharacter|GetAssociatedPlayerPawn) "(TagName=\"StatManager.DamageDealt\")")
      (:OnStatChanged
        (Utilities|FlowControl|Sequence
          (:then_0
            ;[1 1 1] Reactive
            (if (> _statvalue 0)
              (Utilities|FlowControl|DoOnce false
                (:Completed
                  (bind _returnvalue (PartModifier|SpawnSystem|SpawnSystemAttachedPartModifier "/BRCosmetics/Characters/Player/Male/Medium/Bodies/M_MED_Sleek_Glasses_Jacket/FX/NS_Sleek_Body_FlowerHit.NS_Sleek_Body_FlowerHit" (CustomCharacter|GetSkeletalMeshForPartOnPlayerPawn) "root" "0, 0, 0" "0, 0, 0" "KeepRelativeOffset" true false)
                    (:Success
                      (Components|Activation|Activate _returnvalue false)
                      (Utilities|FlowControl|RetriggerableDelay 0.5)
                      (CallFunction|ResetFX)))))))
          (:then_1
            ;[1 1 1] Aura
            (if (>= _statvalue 200)
              (Components|Activation|Activate (CallFunction|GetIdleVFXComponent) false))))))))

(event Custom|ResetFX
  ;[1 1 1] Reactive
  (Utilities|FlowControl|DoOnce false
    (:Completed
      (bind _returnvalue (PartModifier|SpawnSystem|SpawnSystemAttachedPartModifier "/BRCosmetics/Characters/Player/Male/Medium/Bodies/M_MED_Sleek_Glasses_Jacket/FX/NS_Sleek_Body_FlowerHit.NS_Sleek_Body_FlowerHit" (CustomCharacter|GetSkeletalMeshForPartOnPlayerPawn) "root" "0, 0, 0" "0, 0, 0" "KeepRelativeOffset" true false)
        (:Success
          (Components|Activation|Activate _returnvalue false)
          (Utilities|FlowControl|RetriggerableDelay 0.5)
          (CallFunction|ResetFX))))))

(fn ConstructionScript
  (|CallParentFunction))
