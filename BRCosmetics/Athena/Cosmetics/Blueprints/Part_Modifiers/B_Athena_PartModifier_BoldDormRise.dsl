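; Part modifier for the BoldDormRise cosmetic that makes its idle FX grow more reactive as the player gets eliminations.
; On post res-in, gameplay restores the stage from the current kill stat, clamped to the max, then watches that stat, while frontend display uses timed stage previews and resets.
; TriggerReactivityFX spawns a Niagara burst unless the head part is hidden, and a timeline blends the "Reactivity" parameter between stages or reverses it for resets.
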(event EventOnDayPhaseChanged
  (params (CurrentDayPhase :type EFortDayPhase%20Enum :default "Morning") (PreviousDayPhase :type EFortDayPhase%20Enum :default "Morning") (bAtCreation :type Boolean)))

(event EventOnNewPlayerWeaponType
  (params (NewType :type EFortItemType%20Enum :default "WorldItem") (OldType :type EFortItemType%20Enum :default "WorldItem"))
  (|CallParentFunction NewType OldType) ; This node is disabled and will not be called.\nDrag off pins to build functionality.
)

(event EventOnWeaponChanged
  (params (NewWeapon :type Fort%20Weapon%20Object%20Reference) (OldWeapon :type Fort%20Weapon%20Object%20Reference))
  (|CallParentFunction NewWeapon OldWeapon) ; This node is disabled and will not be called.\nDrag off pins to build functionality.
)

(event EventOnWeaponFired)

(event EventOnNewWeapons
  (params (WeaponsAsItems :type Array%20of%20Fort%20Item%20Object%20References)))

(event EventOnCurieStateAttached
  (params (Tag :type Gameplay%20Tag%20Structure%20%28by%20ref%29)))

(event EventOnCurieStateDetached
  (params (Tag :type Gameplay%20Tag%20Structure%20%28by%20ref%29)))

(event EventOnBeginSkydiving
  (|CallParentFunction) ; This node is disabled and will not be called.\nDrag off pins to build functionality.
)

(event EventOnEndSkydiving
  (|CallParentFunction) ; This node is disabled and will not be called.\nDrag off pins to build functionality.
)

(event EventOnEnterVehicle
  (params (VehicleEntered :type Fort%20Vehicle%20Interface%20Interface%20%28by%20ref%29))
  (|CallParentFunction VehicleEntered) ; This node is disabled and will not be called.\nDrag off pins to build functionality.
)

(event EventOnExitVehicle
  (|CallParentFunction) ; This node is disabled and will not be called.\nDrag off pins to build functionality.
)

(event EventOnTargetingOrFirstPersonCameraChanged)

(event EventOnHitPawn
  (params (SelfActor :type Actor%20Object%20Reference) (OtherActor :type Actor%20Object%20Reference) (NormalImpulse :type Vector :default "0, 0, 0") (Hit :type Hit%20Result%20Structure%20%28by%20ref%29)))

(event EventBeginPlay
  (|CallParentFunction))

(event EventOnPostResIn
  (bind _returnvalue (CustomCharacter|GetAssociatedPlayerPawn))
  (|CallParentFunction)
  ; Old Initialization
  (if (Reactive|Frontend|IsCosmeticObjectDisplayedinFrontendContext 0)
    ;[1 0.5169 0.1266 0.5] FE Reactivity
    (Utilities|FlowControl|RetriggerableDelay (Variables|Default|GetTimeBetween_FE_ReactivityLoops))
    (if (< (Variables|Default|GetCurrentStage) (Variables|Default|GetMaxStages))
      (CallFunction|TriggerReactivityFX false))
    (CallFunction|TriggerReactivityFX true)
    (Utilities|FlowControl|Delay 2.0)
    (else
      ;[1 0.1535 0.1568 0.5] Gameplay reactivity
      ;[0.1536 0.7244 0.9219 0.5 13] Checking if the player has a killcount >0 when loading the BP (relevant for items that can unload/reload the skin during the match, like the Wingsuit)
      (bind _returnvalue_2 (Class|FortCosmeticFunctionLibrary|GetCosmeticStatValue _returnvalue "(TagName=\"StatManager.AthenaKills\")"))
      (if (< _returnvalue_2 0)
        (Variables|Default|SetCurrentStage 0.0)
        (elif (> _returnvalue_2 (Variables|Default|GetMaxStages))
          (Variables|Default|SetCurrentStage (Variables|Default|GetMaxStages))
          (else
            (Variables|Default|SetCurrentStage (Math|Conversions|ToFloat(Integer) _returnvalue_2)))))
      (Effects|Components|ParticleSystem|SetFloatParameter (Variables|Default|GetIdleFX) "Reactivity" (/ (Variables|Default|GetCurrentStage) (Variables|Default|GetMaxStages)))
      (Cosmetics|WatchCosmeticStat _returnvalue "(TagName=\"StatManager.AthenaKills\")" 0 false
        (:OnStatChanged
          (if (< (Variables|Default|GetCurrentStage) (Variables|Default|GetMaxStages))
            (CallFunction|TriggerReactivityFX false)))))))

(event Custom|TriggerReactivityFX
  (params (Reset_FE_Reactivity :type Boolean))
  (if Reset_FE_Reactivity
    (Variables|Default|SetReactivityReset true)
    (bind (_direction _reactivity) (Timeline|Timeline ReverseFromEnd)
      (:Update
        (if (Variables|Default|GetReactivityReset)
          ;[1 0.5601 0.1796 0.5] Reset FE Reactivity
          (Effects|Components|ParticleSystem|SetFloatParameter (Variables|Default|GetIdleFX) "Reactivity" _reactivity)
          (else
            (Effects|Components|ParticleSystem|SetFloatParameter (Variables|Default|GetIdleFX) "Reactivity" (+ (* _reactivity (/ 1.0 (Variables|Default|GetMaxStages))) (/ (- (Variables|Default|GetCurrentStage) 1.0) (Variables|Default|GetMaxStages)))))))
      (:Finished
        ;[1 0.5601 0.1796 0.5] Reset FE Reactivity
        (if (Variables|Default|GetReactivityReset)
          (Variables|Default|SetReactivityReset false)
          (Variables|Default|SetCurrentStage 0.0))))
    (else
      (Math|Float|IncrementFloat (Variables|Default|GetCurrentStage)
        (:%20%20
          (Utilities|FlowControl|Sequence
            (:then_0
              ;[0.5625 0.01771 0.0204 0.5] If Head CharacterPart is hidden, do not spawn Reactivity trigger FX
              (if (not (Variables|HiddenPartCheck|Advanced|GetPartHiddenHead))
                ;[0.5888 1 0.3542 0.5] Reactivity trigger Burst FX
                (PartModifier|SpawnSystem|SpawnSystemAttachedPartModifier "/BRCosmetics/Characters/Player/Male/Medium/Bodies/M_MED_BoldDormRise_Frock/FX/NS_Character_BoldDormRise_Reactive_Burst.NS_Character_BoldDormRise_Reactive_Burst" (CustomCharacter|GetSkeletalMeshForPartOnPlayerPawn))))
            (:then_1
              (Timeline|Timeline PlayFromStart))))))))

(fn ConstructionScript
  (|CallParentFunction))
