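; Cosmetic part modifier that drives a level-based "Virus" material effect on the player's part mesh from the Season 11 level stat.
; The stat watcher keeps the highest value seen and pushes current, start and end levels to the materials.
; A gated branch stamps game time for levels strictly inside that range, intended only for real changes.
; A workaround timer reapplies parameters every 0.1 seconds, with a one-second delayed clear after res-in.
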(event EventOnPostResIn
  (|CallParentFunction)
  (CallFunction|ResetSafetyGate)
  (Utilities|FlowControl|Sequence
    (:then_0
      (Utilities|FlowControl|Delay 1.0)
      (Utilities|Time|ClearandInvalidateTimerbyHandle (Variables|Default|GetHackyRefreshTimerHandle)))
    (:then_1
      (bind (_statname _statvalue) (Cosmetics|WatchCosmeticStat (CustomCharacter|GetAssociatedPlayerPawn) "(TagName=\"StatManager.Season.SeasonLevel11\")")
        (:OnStatChanged
          (bind _output_get (Variables|Default|SetLastValue (Math|Integer|Max(Integer) _statvalue (Variables|Default|GetLastValue))))
          (Utilities|FlowControl|Sequence
            (:then_0
              (CallFunction|ApplyVirusParam _output_get)
              (Utilities|FlowControl|DoOnce false
                (:Completed
                  (bind _returnvalue (Utilities|Time|SetTimerbyEvent (delegate Custom|HackyWorkaroundRefreshTimerGuy) 0.1 true))
                  (Variables|Default|SetHackyRefreshTimerHandle _returnvalue))))
            (:then_1
              (Utilities|FlowControl|Gate true
                (:Exit
                  (if (and (> _output_get (Variables|Default|GetStartLevel)) (< _output_get (Variables|Default|GetEndLevel)))
                    (CallFunction|OnActualStatChange)))))))))))

(event Custom|HackyWorkaroundRefreshTimerGuy
  (Utilities|FlowControl|Sequence
    (:then_0
      (CallFunction|ApplyVirusParam (stale EventOnPostResIn _output_get))
      (Utilities|FlowControl|DoOnce false
        (:Completed
          (bind _returnvalue (Utilities|Time|SetTimerbyEvent (delegate Custom|HackyWorkaroundRefreshTimerGuy) 0.1 true))
          (Variables|Default|SetHackyRefreshTimerHandle _returnvalue))))
    (:then_1
      (Utilities|FlowControl|Gate true
        (:Exit
          (if (and (> (stale EventOnPostResIn _output_get) (Variables|Default|GetStartLevel)) (< (stale EventOnPostResIn _output_get) (Variables|Default|GetEndLevel)))
            (CallFunction|OnActualStatChange)))))))

(event Custom|ResetSafetyGate
  (Utilities|FlowControl|DoOnce false
    (:Completed
      (bind _returnvalue (Utilities|Time|SetTimerbyEvent (delegate Custom|HackyWorkaroundRefreshTimerGuy) 0.1 true))
      (Variables|Default|SetHackyRefreshTimerHandle _returnvalue))))

(event Custom|OnActualStatChange
  ;[1 1 1] This should only fire off when the stat changes - use this event!
  (Rendering|Material|SetScalarParameterValueonMaterials (CustomCharacter|GetSkeletalMeshForPartOnPlayerPawn) "Virus_GameTimeFromBP" (Utilities|Time|GetGameTimeinSeconds)))

(fn ApplyVirusParam
  (params (InInt :type Integer))
  (bind _returnvalue (CustomCharacter|GetSkeletalMeshForPartOnPlayerPawn))
  (Rendering|Material|SetScalarParameterValueonMaterials _returnvalue "Virus_CurrentLevel" (Math|Conversions|ToFloat(Integer) InInt))
  (Rendering|Material|SetScalarParameterValueonMaterials _returnvalue "Virus_StartAtLevel" (Math|Conversions|ToFloat(Integer) (Variables|Default|GetStartLevel)))
  (Rendering|Material|SetScalarParameterValueonMaterials _returnvalue "Virus_StopAtLevel" (Math|Conversions|ToFloat(Integer) (Variables|Default|GetEndLevel))))

(fn ConstructionScript
  (|CallParentFunction)
  (Rendering|Material|SetScalarParameterValueonMaterials (CustomCharacter|GetSkeletalMeshForPartOnPlayerPawn) "Virus_GameTimeFromBP" -99999.0))
