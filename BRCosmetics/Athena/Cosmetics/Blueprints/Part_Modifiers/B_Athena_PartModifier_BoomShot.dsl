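; BoomShot character part modifier that adjusts material shading for one progressive style of the cosmetic in the front end.
; After the parent post-res-in logic, it checks the loadout's Progressive variant channel for Stage2, the second of two listed stages.
; When matched, it sets CRTN_LightLevels on a dynamic instance of each material on the part's skeletal mesh, choosing the value by part type, with the head lower than most parts.
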(event EventOnPostResIn
  (bind _returnvalue (CustomCharacter|GetSkeletalMeshForPartOnPlayerPawn))
  ; Adjust shading in FE
  (|CallParentFunction)
  (if (CustomCosmeticModifier|IsinFrontEnd)
    (bind _index (Class|FortCosmeticFunctionLibrary|GetVariantChannelActiveIndexForSlot "(TagName=\"LoadoutSlot.Character\")" (CosmeticLoadout|GetCosmeticLoadout) "(TagName=\"Cosmetics.Variant.Channel.Progressive\")" (Utilities|Array|MakeArray "(TagName=\"Cosmetics.Variant.Property.Stage1\")" "(TagName=\"Cosmetics.Variant.Property.Stage2\")")))
    (if (== _index 1)
      (for (_ _array_index) (Rendering|Material|GetMaterials _returnvalue)
        (Rendering|Material|SetScalarParameterValue (Class|BTechArtGlobalLibrary|GetDynamicMaterialInstance _returnvalue _array_index) "CRTN_LightLevels" (Utilities|Select :Index (CustomCharacter|GetCharacterPartType) :Head 1.4 :Body 1.8 :Hat 1.8 :Backpack 1.8 :MiscOrTail 1.8 :Face 1.8 :Gameplay 1.8 :ExtraPart 1.8 :Gameplay2 0.0) :!class /Script/Engine.MaterialInstanceDynamic)))))

(fn ConstructionScript
  (|CallParentFunction))
