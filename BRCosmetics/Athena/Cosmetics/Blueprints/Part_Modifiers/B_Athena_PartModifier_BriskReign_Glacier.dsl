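; Part modifier for the BriskReign Glacier cosmetic body that plays its reactive Niagara effect in the frontend or in response to kills.
; After res-in, if frontend reactivity should display, it spawns the effect once after a short delay; a trailing delay leads to nothing.
; Otherwise it watches the pawn's AthenaKills stat and spawns the same effect, attached to the part mesh's pelvis, whenever that stat changes.
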(event EventOnPostResIn
  (|CallParentFunction)
  (if (Reactive|Frontend|ShouldDisplayFrontendReactivity 0 "Body")
    (Utilities|FlowControl|Delay 2.0)
    (Niagara|SpawnSystemAttached "/BRCosmetics/Characters/Player/Male/Medium/Bodies/M_MED_BriskReign_Glacier/FX/NS_Character_BriskReign_Glacier_Reactive.NS_Character_BriskReign_Glacier_Reactive" (CustomCharacter|GetSkeletalMeshForPartOnPlayerPawn) "pelvis")
    (Utilities|FlowControl|Delay 5.0)
    (else
      (Cosmetics|WatchCosmeticStat (CustomCharacter|GetAssociatedPlayerPawn) "(TagName=\"StatManager.AthenaKills\")" 0 false
        (:OnStatChanged
          (Niagara|SpawnSystemAttached "/BRCosmetics/Characters/Player/Male/Medium/Bodies/M_MED_BriskReign_Glacier/FX/NS_Character_BriskReign_Glacier_Reactive.NS_Character_BriskReign_Glacier_Reactive" (CustomCharacter|GetSkeletalMeshForPartOnPlayerPawn) "pelvis"))))))

(fn ConstructionScript
  (|CallParentFunction))
