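; Cosmetic part modifier for the BriskReign Glacier outfit that plays a reactive material and Niagara effect when a stat changes to a positive value.
; The body material gets a start time, while body-part FX spawn impulse and timeline-driven reactive systems, with a timer updating head/hand hidden flags.
; Retriggers rewind or skip the running timeline and add another impulse.
; Frontend and preview replay on delays; hologram material overrides postpone FX.
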(event EventOnDayPhaseChanged
  (params (CurrentDayPhase :type EFortDayPhase%20Enum :default "Morning") (PreviousDayPhase :type EFortDayPhase%20Enum :default "Morning") (bAtCreation :type Boolean)))

(event Custom|OnMaterialOverrideApplied_Event
  (params (AffectedPawn :type Fort%20Player%20Pawn%20Object%20Reference) (AppliedOverride_OverrideId :type Guid%20Structure) (AppliedOverride_Material :type Material%20Interface%20Soft%20Object%20Reference))
  ; Events
  (Variables|Default|SetIs%20Hologram true))

(event EventOnNewPlayerWeaponType
  (params (NewType :type EFortItemType%20Enum :default "WorldItem") (OldType :type EFortItemType%20Enum :default "WorldItem")))

(event Custom|Setup%20Reactivity%20Material
  (bind _bodymesh (Variables|Default|GetBodyMesh))
  (bind _paramreactivityfrontend (Variables|Default|GetParamReactivityFrontEnd))
  ; Reactivity Material
  (Variables|Default|SetFaceMesh (Customization|GetSkeletalMeshforPartType (Variables|Default|GetPlayerPawn) "Face" :!class /Script/FortniteGame.FortPlayerPawn))
  (Variables|Default|SetBackpackMesh (Customization|GetSkeletalMeshforPartType (Variables|Default|GetPlayerPawn) "Backpack" :!class /Script/FortniteGame.FortPlayerPawn))
  (bind _output_get (Variables|Default|SetIsFrontEnd (Reactive|Frontend|IsCosmeticObjectDisplayedinFrontendContext 0)))
  (if _output_get
    (Variables|Default|SetParamReactivityDuration "Reactivity - Duration [Frontend]")
    (Rendering|Material|SetScalarParameterValueonMaterials _bodymesh _paramreactivityfrontend 1.0)
    (else
      (Variables|Default|SetParamReactivityDuration "Reactivity - Duration [Gameplay]")
      (Rendering|Material|SetScalarParameterValueonMaterials _bodymesh _paramreactivityfrontend 0.0)))
  (Variables|Default|SetDuration 0.0)
  (for _array_element (Utilities|Array|MakeArray (Variables|Default|GetBodyMesh) (Variables|Default|GetFaceMesh) (Variables|Default|GetBackpackMesh))
    (bind _asmaterial_instance_dynamic (Utilities|Casting|CastToMaterialInstanceDynamic (Rendering|Material|GetMaterial _array_element 0)))
    (bind _returnvalue (Rendering|Material|GetVectorParameterValue _asmaterial_instance_dynamic (Variables|Default|GetParamReactivityDuration) :!class /Script/Engine.MaterialInstanceDynamic))
    (bind _returnvalue_r (Math|Color|BreakColor _returnvalue))
    (Variables|Default|SetDuration (Math|Float|Max(Float) _returnvalue_r (Variables|Default|GetDuration)))))

(event Custom|OnMaterialOverrideCleared_Event
  (params (AffectedPawn :type Fort%20Player%20Pawn%20Object%20Reference))
  ; Events
  (Variables|Default|SetIs%20Hologram false))

(event EventOnWeaponFired)

(event EventOnPostResIn
  (bind _returnvalue (CustomCharacter|GetAssociatedPlayerPawn))
  ; Events
  (|CallParentFunction)
  (bind _output_get (Variables|Default|SetPlayerPawn (CustomCharacter|GetAssociatedPlayerPawn)))
  (Variables|Default|SetBodyMesh (Customization|GetSkeletalMeshforPartType _output_get "Body" :!class /Script/FortniteGame.FortPlayerPawn))
  (Utilities|FlowControl|Sequence
    (:then_0
      (EventDispatchers|AssignOnMaterialOverrideApplied _returnvalue (delegate Custom|OnMaterialOverrideApplied_Event))
      (EventDispatchers|AssignOnMaterialOverrideCleared _returnvalue (delegate Custom|OnMaterialOverrideCleared_Event)))
    (:then_1
      (CallFunction|SetupReactivityMaterial))
    (:then_2
      (if (and (Utilities|Enum|Equal(Enum) (CustomCharacter|GetCharacterPartType) "Body") (CallFunction|ShouldDisplayFrontendReactivity))
        (Utilities|FlowControl|Delay 2.0)
        (CallFunction|TriggerReactivityMaterial)
        (CallFunction|TriggerReactivityFX)))))

(event EventOnNewWeapons
  (params (WeaponsAsItems :type Array%20of%20Fort%20Item%20Object%20References)))

(event Custom|Trigger%20Reactivity%20Material
  ; Reactivity Material
  (Rendering|Material|SetScalarParameterValueonMaterials (Variables|Default|GetBodyMesh) (Variables|Default|GetParamReactivityStartTime) (Utilities|Time|GetGameTimeinSeconds)))

(event EventOnCurieStateAttached
  (params (Tag :type Gameplay%20Tag%20Structure%20%28by%20ref%29)))

(event EventOnCurieStateDetached
  (params (Tag :type Gameplay%20Tag%20Structure%20%28by%20ref%29)))

(event EventOnBeginSkydiving
  (|CallParentFunction) ; This node is disabled and will not be called.\nDrag off pins to build functionality.
)

(event EventOnEndSkydiving
  (|CallParentFunction) ; This node is disabled and will not be called.\nDrag off pins to build functionality.
)

(event EventOnStatChanged
  (params (State%20Name :type Name :default "None") (Stat%20Value :type Integer))
  ; Events
  (|CallParentFunction State%20Name Stat%20Value)
  (if (> Stat%20Value 0)
    (Utilities|FlowControl|Sequence
      (:then_0
        (CallFunction|TriggerReactivityMaterial))
      (:then_1
        (if (Utilities|Enum|Equal(Enum) (CustomCharacter|GetCharacterPartType) "Body")
          (CallFunction|TriggerReactivityFX))))))

(event EventOnEnterVehicle
  (params (VehicleEntered :type Fort%20Vehicle%20Interface%20Interface%20%28by%20ref%29))
  (|CallParentFunction VehicleEntered) ; This node is disabled and will not be called.\nDrag off pins to build functionality.
)

(event Custom|Trigger%20Reactivity%20FX
  (bind _returnvalue (CustomCharacter|GetSkeletalMeshForPartOnPlayerPawn))
  ; Reactivity FX
  (if (Variables|Default|GetCanSpawnReactivity)
    (if (not (Variables|Default|GetIs%20Hologram))
      (Variables|Default|SetCanSpawnReactivity false)
      (Utilities|FlowControl|Sequence
        (:then_0
          (if (Components|Timeline|IsPlaying (Variables|B_Athena_PartModifier_BriskReign_GlacierHit|GetTL%20Reactivity))
            (CallFunction|RetriggerReactivity)
            (else
              (bind _returnvalue_2 (Niagara|SpawnSystemAttached "/BRCosmetics/Characters/Player/Male/Medium/Bodies/M_MED_BriskReign_Glacier/FX/NS_Character_BriskReign_Glacier_Impulse.NS_Character_BriskReign_Glacier_Impulse" _returnvalue "pelvis" "0, 0, 0" "0, 0, 0" "KeepRelativeOffset" true))
              (Variables|Default|SetNS%20Impulse _returnvalue_2)
              (bind _returnvalue_3 (Niagara|SpawnSystemAttached "/BRCosmetics/Characters/Player/Male/Medium/Bodies/M_MED_BriskReign_Glacier/FX/NS_Character_BriskReign_Glacier_Reactive.NS_Character_BriskReign_Glacier_Reactive" _returnvalue "pelvis" "0, 0, 0" "0, 0, 0" "KeepRelativeOffset" true))
              (Variables|Default|SetNS%20Reactive _returnvalue_3)
              (CallFunction|CheckforAGIDs)
              (bind _returnvalue_4 (Utilities|Time|SetTimerbyEvent (EventDispatchers|CreateEvent :!event Custom|CheckForAGIDs) 0.1 true))
              (Variables|Default|SetTimerHandle _returnvalue_4)
              (bind (_direction _alpha) (Timeline|TL%20Reactivity PlayFromStart)
                (:Update
                  (Niagara|SetNiagaraVariable(Float) (Variables|Default|GetNS%20Reactive) "Alpha" _alpha))
                (:Finished
                  (Components|DestroyComponent (Variables|Default|GetNS%20Reactive))
                  (Utilities|Time|ClearandInvalidateTimerbyHandle (Variables|Default|GetTimerHandle))
                  (if (or (Reactive|Frontend|IsCosmeticObjectDisplayedinFrontendContext 0) (Reactive|Frontend|IsCosmeticObjectDisplayedasPreview 0))
                    (Utilities|FlowControl|Delay 3.0)
                    (CallFunction|TriggerReactivityMaterial)
                    (CallFunction|TriggerReactivityFX)))))))
        (:then_1
          ;; Added this to prevent a bug triggering multiple spawns after equipping & unequipping AGIDS
          (Utilities|FlowControl|Delay 0.1)
          (Variables|Default|SetCanSpawnReactivity true)))
      (else
        (Utilities|FlowControl|Delay 3.0)
        (CallFunction|TriggerReactivityFX)))))

(event EventOnExitVehicle
  (|CallParentFunction) ; This node is disabled and will not be called.\nDrag off pins to build functionality.
)

(event EventOnTargetingOrFirstPersonCameraChanged)

(event EventOnHitPawn
  (params (SelfActor :type Actor%20Object%20Reference) (OtherActor :type Actor%20Object%20Reference) (NormalImpulse :type Vector :default "0, 0, 0") (Hit :type Hit%20Result%20Structure%20%28by%20ref%29)))

(event Custom|Retrigger%20Reactivity
  ; Reactivity FX
  (Components|Timeline|SetPlaybackPosition (Variables|B_Athena_PartModifier_BriskReign_GlacierHit|GetTL%20Reactivity) 0.75 false true)
  (bind _returnvalue (Niagara|SpawnSystemAttached "/BRCosmetics/Characters/Player/Male/Medium/Bodies/M_MED_BriskReign_Glacier/FX/NS_Character_BriskReign_Glacier_Impulse.NS_Character_BriskReign_Glacier_Impulse" (CustomCharacter|GetSkeletalMeshForPartOnPlayerPawn) "pelvis" "0, 0, 0" "0, 0, 0" "KeepRelativeOffset" true))
  (Variables|Default|SetNS%20Impulse _returnvalue))

(event Custom|CheckForAGIDs
  (bind _ns_reactive (Variables|Default|GetNS%20Reactive))
  ; Reactivity FX
  (Niagara|SetNiagaraVariable(Bool) _ns_reactive "bIsHeadHidden" (Variables|HiddenPartCheck|GetbIsHeadHidden))
  (Niagara|SetNiagaraVariable(Bool) _ns_reactive "bIsHandHidden" (Variables|HiddenPartCheck|GetbIsHandHidden)))

(fn ConstructionScript
  (|CallParentFunction))
