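; Cosmetic part modifier for the Brutal Burglar backpack that drives kill-reactive material parameters on its skeletal mesh.
; Construction seeds the SL_ game time and start and stop level parameters.
; After post-res-in, it watches the Athena kills stat, writing BP_KillCount up to a cap, beyond which it writes BP_GetTime.
; In frontend display, a looping timer fakes kills, cycling the count modulo seven and timestamping non-zero steps.
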(event Custom|FE_Display
  (bind _returnvalue (CustomCharacter|GetSkeletalMeshForPartOnPlayerPawn))
  (bind _output_get (Variables|Default|SetFrontEnd_KillCount (+ (Variables|Default|GetFrontEnd_KillCount) 1.0)))
  (Rendering|Material|SetScalarParameterValueonMaterials _returnvalue "BP_KillCount" (- _output_get (* (Math|Float|Floor (/ _output_get 7.0)) 7)))
  (if (not (== (- _output_get (* (Math|Float|Floor (/ _output_get 7.0)) 7)) 0))
    (Rendering|Material|SetScalarParameterValueonMaterials _returnvalue "BP_GetTime" (Utilities|Time|GetGameTimeinSeconds))))

(event EventOnPostResIn
  (bind _returnvalue (CustomCharacter|GetSkeletalMeshForPartOnPlayerPawn))
  ;[1 1 1] Reactivity Logic
  (|CallParentFunction)
  (Utilities|FlowControl|Sequence
    (:then_0
      (if (Reactive|Frontend|IsCosmeticObjectDisplayedinFrontendContext 0)
        (Utilities|Time|SetTimerbyEvent (delegate Custom|FE_Display) 4.0 true)))
    (:then_1
      (bind (_statname _statvalue) (Cosmetics|WatchCosmeticStat (CustomCharacter|GetAssociatedPlayerPawn) "(TagName=\"StatManager.AthenaKills\")" 0 false)
        (:OnStatChanged
          (if (<= _statvalue 6)
            (Rendering|Material|SetScalarParameterValueonMaterials _returnvalue "BP_KillCount" (Math|Conversions|ToFloat(Integer) _statvalue))
            (Rendering|Material|SetScalarParameterValueonMaterials _returnvalue "BP_GetTime" (Utilities|Time|GetGameTimeinSeconds))))))))

(fn ConstructionScript
  (bind _returnvalue (CustomCharacter|GetSkeletalMeshForPartOnPlayerPawn))
  (|CallParentFunction)
  (Rendering|Material|SetScalarParameterValueonMaterials (CustomCharacter|GetSkeletalMeshForPartOnPlayerPawn) "SL_GameTimeFromBP" -99999.0)
  (Rendering|Material|SetScalarParameterValueonMaterials _returnvalue "SL_StartAtLevel" (Math|Conversions|ToFloat(Integer) (Variables|Default|GetKillCount)))
  (Rendering|Material|SetScalarParameterValueonMaterials _returnvalue "SL_StopAtLevel" (Math|Conversions|ToFloat(Integer) (Variables|Default|GetSL_EndLevel))))
