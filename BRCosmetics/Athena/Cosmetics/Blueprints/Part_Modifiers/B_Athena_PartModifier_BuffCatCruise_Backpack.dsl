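; Part modifier Blueprint for the BuffCatCruise back bling that makes its effects react to the player jumping and landing.
; On jump, it sets the bJumpReactivity parameter on the idle VFX component to true.
; On landing, it clears that parameter and spawns the Niagara splash system attached to the part's skeletal mesh on the player pawn.
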(event EventOnPawnLanded
  (Effects|Components|ParticleSystem|SetBoolParameter (CallFunction|GetIdleVFXComponent) "bJumpReactivity" false)
  ;[1 1 1] Splash on landing
  (Niagara|SpawnSystemAttached "/BRCosmetics/Accessories/FORT_Backpacks/Backpack_M_MED_BuffCatCruise/FX/NS_BuffCatCruise_Splash.NS_BuffCatCruise_Splash" (CustomCharacter|GetSkeletalMeshForPartOnPlayerPawn) "None" "0, 0, 0" "0, 0, 0" "KeepRelativeOffset" true))

(event EventOnPawnJumped
  (Effects|Components|ParticleSystem|SetBoolParameter (CallFunction|GetIdleVFXComponent) "bJumpReactivity" true))

(fn ConstructionScript
  (|CallParentFunction))
