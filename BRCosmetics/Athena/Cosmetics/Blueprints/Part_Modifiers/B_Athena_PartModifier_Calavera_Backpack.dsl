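; Part modifier for the Calavera backpack cosmetic that, when flagged reactive, animates its idle VFX and material glow.
; In frontend display, after a short delay, looping timelines drive particle dissolve and alpha plus emissive and pulse parameters on the mesh.
; Otherwise it watches the pawn's kill stat, and a positive change plays an attached sound and runs in-game particle and emissive timelines.
; A debug "kill" print remains in the kill path.
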(event EventOnPostResIn
  (bind _returnvalue (CustomCharacter|GetSkeletalMeshForPartOnPlayerPawn))
  (bind _returnvalue_2 (CallFunction|GetIdleVFXComponent))
  (bind _returnvalue_3 (CallFunction|GetIdleVFXComponent))
  (|CallParentFunction)
  (if (Variables|Default|GetIs%20Reactive)
    (if (Reactive|Frontend|ShouldDisplayFrontendReactivity 0 "Backpack")
      (Utilities|FlowControl|DoOnce false
        (:Completed
          (Utilities|FlowControl|Delay 1.0)
          (Utilities|FlowControl|Sequence
            (:then_0
              (bind (_direction _dissolve _alpha) (Timeline|FX_TimeLine_FE_Looping PlayFromStart)
                (:Update
                  (Effects|Components|ParticleSystem|SetFloatParameter _returnvalue_2 "Dissolve Progress" _dissolve)
                  (Effects|Components|ParticleSystem|SetFloatParameter _returnvalue_2 "Global Alpha" _alpha))))
            (:then_1
              (bind (_direction_2 _vibrantamount _pulsealpha) (Timeline|Vibrant%20FE%20Loop PlayFromStart)
                (:Update
                  (Rendering|Material|SetScalarParameterValueonMaterials _returnvalue "VibrantEmissiveAmount" _vibrantamount)
                  (Rendering|Material|SetScalarParameterValueonMaterials _returnvalue "PulseVisibility" _pulsealpha)))))))
      (else
        (Utilities|FlowControl|DoOnce false
          (:Completed
            (bind (_statname _statvalue) (Cosmetics|WatchCosmeticStat (CustomCharacter|GetAssociatedPlayerPawn) "(TagName=\"StatManager.AthenaKills\")" 0 false)
              (:OnStatChanged
                (if (> (Math|Conversions|ToFloat(Integer) _statvalue) 0.0)
                  (Development|PrintString "kill" true true "(R=0.333333,G=1.000000,B=0.000000,A=1.000000)" 5.0)
                  (Utilities|FlowControl|Sequence
                    (:then_0
                      ;[0 0 0] Trigger "OnKill" Audio
                      (CallFunction|TriggerAudio)
                      (bind (_direction_3 _dissolve_2 _alpha_2) (Timeline|FX_TimeLine_Game PlayFromStart)
                        (:Update
                          (Effects|Components|ParticleSystem|SetFloatParameter _returnvalue_3 "Dissolve Progress" _dissolve_2)
                          (Effects|Components|ParticleSystem|SetFloatParameter _returnvalue_3 "Global Alpha" _alpha_2))
                        (:Finished
                          (Utilities|FlowControl|DoOnce false))))
                    (:then_1
                      (bind (_direction_4 _vibrantamount_2) (Timeline|Vibrant%20Amount PlayFromStart)
                        (:Update
                          (Rendering|Material|SetScalarParameterValueonMaterials (CustomCharacter|GetSkeletalMeshForPartOnPlayerPawn) "VibrantEmissiveAmount" _vibrantamount_2))))))))))))))

(event Custom|TriggerAudio
  ;[0 0 0] Audio
  (Audio|SpawnSoundAttached (Variables|Default|GetSound) (CustomCharacter|GetSkeletalMeshForPartOnPlayerPawn)))

(fn ConstructionScript
  (|CallParentFunction))
