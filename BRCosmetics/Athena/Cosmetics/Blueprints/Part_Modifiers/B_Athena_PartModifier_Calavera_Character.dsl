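; Character part modifier for the Calavera cosmetic that, when reactive, drives emissive materials and idle particle dissolve.
; Tagged part meshes are cached after customization; the front end plays delayed showcase timelines.
; In game, it watches the player's kill stat: increases trigger attached audio and a vibrant color pulse, while kill count selects tiered dissolve timelines.
; Overlapping pulses pass the last emissive amount into a second start timeline.
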(event EventOnPostResIn
  (bind _returnvalue (CallFunction|GetIdleVFXComponent))
  (bind _returnvalue_2 (CallFunction|GetIdleVFXComponent))
  (bind _calavera_mesh_components (Variables|Default|GetCalavera%20Mesh%20Components))
  (bind _calavera_mesh_components_2 (Variables|Default|GetCalavera%20Mesh%20Components))
  (bind _calavera_mesh_components_3 (Variables|Default|GetCalavera%20Mesh%20Components))
  (bind _calavera_mesh_components_4 (Variables|Default|GetCalavera%20Mesh%20Components))
  (|CallParentFunction)
  (if (Variables|Default|GetIs%20Reactive)
    (if (CustomCosmeticModifier|IsinFrontEnd)
      (Rendering|Material|SetScalarParameterValueonMaterials _calavera_mesh_components "VibrantEmissiveMax" 0.8)
      (Utilities|FlowControl|DoOnce false
        (:Completed
          (Utilities|FlowControl|Delay 1.0)
          (Utilities|FlowControl|Sequence
            (:then_0
              (bind (_direction _dissolve _alpha _dissolve_edge_intensity) (Timeline|FX_TimeLine_FE_Looping PlayFromStart)
                (:Update
                  (Effects|Components|ParticleSystem|SetFloatParameter _returnvalue "Dissolve Progress" _dissolve)
                  (Effects|Components|ParticleSystem|SetFloatParameter _returnvalue "Global Alpha" _alpha)
                  (Effects|Components|ParticleSystem|SetFloatParameter _returnvalue "Dissolve Edge Intensity" _dissolve_edge_intensity))))
            (:then_1
              (bind (_direction_2 _vibrantamount _pulsealpha) (Timeline|Vibrant%20FE%20Loop PlayFromStart)
                (:Update
                  (Rendering|Material|SetScalarParameterValueonMaterials _calavera_mesh_components_4 "VibrantEmissiveAmount" _vibrantamount)
                  (Rendering|Material|SetScalarParameterValueonMaterials _calavera_mesh_components_4 "PulseVisibility" _pulsealpha)))))))
      (else
        (Rendering|Material|SetScalarParameterValueonMaterials _calavera_mesh_components "VibrantEmissiveMax" 0.5)
        (bind (_statname _statvalue) (Cosmetics|WatchCosmeticStat (CustomCharacter|GetAssociatedPlayerPawn) "(TagName=\"StatManager.AthenaKills\")" 0 false)
          (:then
            _)
          (:OnStatChanged
            (if (> _statvalue (Variables|Default|GetKill%20Count))
              (Utilities|FlowControl|Sequence
                (:then_0
                  _)
                (:then_1
                  ;[1 1 1] Vibrant colors start
                  (if (Variables|Default|GetIs%20Finished)
                    (bind (_direction_3 _vibrantamount_2) (Timeline|Vibrant%20Start%2001 PlayFromStart)
                      (:Update
                        (Rendering|Material|SetScalarParameterValueonMaterials _calavera_mesh_components_3 "VibrantEmissiveAmount" _vibrantamount_2)
                        (Rendering|Material|SetScalarParameterValueonMaterials _calavera_mesh_components_3 "PulseVisibility" _vibrantamount_2)
                        (Variables|Default|SetLastEmissiveAmount _vibrantamount_2)
                        (Variables|Default|SetIs%20Finished false :!label _label))
                      (:Finished
                        _))
                    (else
                      (Rendering|Material|SetScalarParameterValueonMaterials _calavera_mesh_components_3 "LastAmount" (Variables|Default|GetLastEmissiveAmount))
                      (Utilities|FlowControl|Sequence
                        (:then_0
                          (Timeline|Vibrant%20Start%2001 Stop))
                        (:then_1
                          (bind (_direction_4 _vibrantamount_3) (Timeline|Vibrant%20Start%2002 PlayFromStart)
                            (:Update
                              (Rendering|Material|SetScalarParameterValueonMaterials _calavera_mesh_components_3 "VibrantEmissiveAmount" _vibrantamount_3)
                              (Rendering|Material|SetScalarParameterValueonMaterials _calavera_mesh_components_3 "PulseVisibility" _vibrantamount_3)
                              (goto _label))
                            (:Finished
                              _))))))
                  (Variables|Default|SetIs%20Finished true)
                  (bind _output_get (Variables|Default|SetLastEmissiveAmount 0.0))
                  (Rendering|Material|SetScalarParameterValueonMaterials _calavera_mesh_components_3 "LastAmount" _output_get)
                  ;[1 1 1] Vibrant colors end
                  (bind (_direction_5 _vibrantamount_4) (Timeline|Vibrant%20End PlayFromStart)
                    (:Update
                      (Rendering|Material|SetScalarParameterValueonMaterials _calavera_mesh_components_2 "VibrantEmissiveAmount" _vibrantamount_4)
                      (Rendering|Material|SetScalarParameterValueonMaterials _calavera_mesh_components_2 "PulseVisibility" _vibrantamount_4))))
                (:then_2
                  ;[0 0 0] Reactive "OnKill" Audio
                  (CallFunction|TriggerAudio))))))
        (bind _output_get_2 (Variables|Default|SetKill%20Count _statvalue))
        (switch int _output_get_2
          (:Default
            (bind (_direction_6 _dissolve_time _dissolve_edge_intensity_2) (Timeline|FX_TimeLine_Game03 PlayFromStart)
              (:Update
                (Effects|Components|ParticleSystem|SetFloatParameter _returnvalue_2 "Dissolve Progress" (Math|Float|Lerp 0.0 1.0 _dissolve_time))
                (Effects|Components|ParticleSystem|SetFloatParameter _returnvalue_2 "Dissolve Edge Intensity" _dissolve_edge_intensity_2))))
          (:0
            (Effects|Components|ParticleSystem|SetFloatParameter _returnvalue_2 "Dissolve Progress" 0.0)
            (Rendering|Material|SetScalarParameterValueonMaterials (Variables|Default|GetCalavera%20Mesh%20Components) "VibrantEmissiveAmount" 0.0) ; Turn off colors when kills == 0
          )
          (:1
            (bind (_direction_7 _dissolve_time_2 _dissolve_edge_intensity_3) (Timeline|FX_TimeLine_Game01 PlayFromStart)
              (:Update
                (Effects|Components|ParticleSystem|SetFloatParameter _returnvalue_2 "Dissolve Progress" (Math|Float|Lerp 0.0 0.45 _dissolve_time_2))
                (Effects|Components|ParticleSystem|SetFloatParameter _returnvalue_2 "Dissolve Edge Intensity" _dissolve_edge_intensity_3))))
          (:2
            (Timeline|FX_TimeLine_Game01 PlayFromStart))
          (:3
            (bind (_direction_8 _dissolve_time_3 _dissolve_edge_intensity_4) (Timeline|FX_TimeLine_Game02 PlayFromStart)
              (:Update
                (Effects|Components|ParticleSystem|SetFloatParameter _returnvalue_2 "Dissolve Progress" (Math|Float|Lerp 0.0 0.7 _dissolve_time_3))
                (Effects|Components|ParticleSystem|SetFloatParameter _returnvalue_2 "Dissolve Edge Intensity" _dissolve_edge_intensity_4))))
          (:4
            (Timeline|FX_TimeLine_Game02 PlayFromStart))
          (:5
            (Timeline|FX_TimeLine_Game03 PlayFromStart)))
        (bind (_direction_9 _alpha_2) (Timeline|Timeline PlayFromStart)
          (:Update
            (Effects|Components|ParticleSystem|SetFloatParameter _returnvalue_2 "Global Alpha" _alpha_2)))))))

(event EventOnCharacterCustomizationFinished
  (params (CharacterPartInfo :type Array%20of%20Applied%20Character%20Part%20Info%20Structures))
  ;[1 1 1] Get all reactive character part mesh
  (bind _returnvalue (Class|FortCosmeticFunctionLibrary|FilterPartMeshComponentsbyTag CharacterPartInfo "(TagName=\"CharacterPart.Character.Calavera.Reactive\")"))
  (Variables|Default|SetCalavera%20Mesh%20Components _returnvalue))

(event Custom|TriggerAudio
  ;[0 0 0] Audio
  (Audio|SpawnSoundAttached (Variables|Default|GetSound) (CustomCharacter|GetSkeletalMeshForPartOnPlayerPawn)))

(fn ConstructionScript
  (|CallParentFunction))
