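; Cosmetic part modifier for the Candy Apple pack that syncs its idle VFX with the pack animation's shield visibility.
; After the parent's post res-in logic, it gets the part's skeletal mesh on the player pawn and casts its anim instance to the Candy Apple pack AnimBP.
; It binds to that AnimBP's HideShield event, deactivating the idle VFX component when hidden and activating it when shown.
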(event EventOnPostResIn
  (|CallParentFunction)
  (bind _asm_med_candy_apple_pack_anim_bp (Utilities|Casting|CastToM_MED_Candy_Apple_Pack_AnimBP (Components|SkeletalMesh|GetAnimInstance (Customization|GetSkeletalMeshforPartType (CustomCharacter|GetAssociatedPlayerPawn) (CustomCharacter|GetCharacterPartType) :!class /Script/FortniteGame.FortPlayerPawn))))
  (Default|BindEventtoHideShield _asm_med_candy_apple_pack_anim_bp (delegate Custom|Shield)))

(event Custom|Shield
  (params (isHidden :type Boolean))
  (bind _returnvalue (CallFunction|GetIdleVFXComponent))
  (if isHidden
    (Components|Activation|Deactivate _returnvalue)
    (else
      (Components|Activation|Activate _returnvalue false))))

(fn ConstructionScript)
