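; Cosmetic part modifier for the CanvasPrint body that connects its idle Niagara VFX to the player pawn's Face and Body meshes.
; After res-in it binds both meshes, then looping timers push each anim instance's LOD to the VFX, forcing zero in frontend.
; Frontend and preview displays enable a fake audio loop material parameter.
; Targeting or first-person camera changes pass ADS state to the VFX for locally controlled pawns.
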(event EventOnPostResIn
  (bind _returnvalue (CustomCharacter|GetAssociatedPlayerPawn))
  (bind _returnvalue_2 (Customization|GetSkeletalMeshforPartType _returnvalue "Face" :!class /Script/FortniteGame.FortPlayerPawn))
  (bind _returnvalue_3 (CallFunction|GetIdleVFXComponent))
  (bind _returnvalue_4 (Customization|GetSkeletalMeshforPartType _returnvalue "Body" :!class /Script/FortniteGame.FortPlayerPawn))
  (bind _returnvalue_5 (CallFunction|GetIdleVFXComponent))
  (|CallParentFunction)
  (Utilities|FlowControl|Sequence
    (:then_0
      ; Face Dyn Bones
      (Utilities|IsValid _returnvalue_2
        (:Is%20Valid
          (Utilities|FlowControl|Sequence
            (:then_0
              (Utilities|IsValid _returnvalue_5
                (:Is%20Valid
                  (bind _asniagara_particle_system_component (Utilities|Casting|CastToNiagaraComponent _returnvalue_5))
                  (Niagara|SetNiagaraSkeletalMeshComponent _asniagara_particle_system_component "HeadRef" _returnvalue_2))))
            (:then_1
              ;; LOD Check for Niagara
              (bind _ascustom_character_part_anim_instance (Utilities|Casting|CastToCustomCharacterPartAnimInstance (Components|SkeletalMesh|GetAnimInstance (Customization|GetSkeletalMeshforPartType (CustomCharacter|GetAssociatedPlayerPawn) "Face" :!class /Script/FortniteGame.FortPlayerPawn))))
              (Variables|Default|SetFaceAccAnimInstance _ascustom_character_part_anim_instance)
              (Utilities|Time|SetTimerbyEvent (EventDispatchers|CreateEvent :!event Custom|LODCheckFace) 2.0 true))))))
    (:then_1
      ; Body Dyn Bones
      (Utilities|IsValid _returnvalue_4
        (:Is%20Valid
          (Utilities|FlowControl|Sequence
            (:then_0
              (Utilities|IsValid _returnvalue_3
                (:Is%20Valid
                  (bind _asniagara_particle_system_component_2 (Utilities|Casting|CastToNiagaraComponent _returnvalue_3))
                  (Niagara|SetNiagaraSkeletalMeshComponent _asniagara_particle_system_component_2 "BodyRef" _returnvalue_4))))
            (:then_1
              ;; LOD Check for Niagara
              (bind _ascustom_character_part_anim_instance_2 (Utilities|Casting|CastToCustomCharacterPartAnimInstance (Components|SkeletalMesh|GetAnimInstance (Customization|GetSkeletalMeshforPartType (CustomCharacter|GetAssociatedPlayerPawn) "Body" :!class /Script/FortniteGame.FortPlayerPawn))))
              (Variables|Default|SetBodyAccAnimInstance _ascustom_character_part_anim_instance_2)
              (Utilities|Time|SetTimerbyEvent (EventDispatchers|CreateEvent :!event Custom|LODCheckBody) 2.0 true))))))
    (:then_2
      (if (or (Reactive|Frontend|IsCosmeticObjectDisplayedinFrontendContext 0) (Reactive|Frontend|IsCosmeticObjectDisplayedasPreview 0))
        (Rendering|Material|SetScalarParameterValueonMaterials (CustomCharacter|GetSkeletalMeshForPartOnPlayerPawn) "EnableFrontendFakeAudioLoop" 1.0)))))

(event Custom|LODCheckFace
  (bind _returnvalue (CallFunction|GetIdleVFXComponent))
  ; Face Dyn Bones
  ;;[1 1 1] LOD Check
  (if (Reactive|Frontend|IsCosmeticObjectDisplayedinFrontendContext 0)
    (Effects|Components|ParticleSystem|SetIntParameter _returnvalue "FaceAccLOD" 0)
    (else
      (Effects|Components|ParticleSystem|SetIntParameter _returnvalue "FaceAccLOD" (Variables|/Script/FortniteGame.CustomCharacterPartAnimInstance|GetCurrentLOD (Variables|Default|GetFaceAccAnimInstance))))))

(event Custom|LODCheckBody
  (bind _returnvalue (CallFunction|GetIdleVFXComponent))
  ; Body Dyn Bones
  ;;[1 1 1] LOD Check
  (if (Reactive|Frontend|IsCosmeticObjectDisplayedinFrontendContext 0)
    (Effects|Components|ParticleSystem|SetIntParameter _returnvalue "BodyAccLOD" 0)
    (else
      (Effects|Components|ParticleSystem|SetIntParameter _returnvalue "BodyAccLOD" (Variables|/Script/FortniteGame.CustomCharacterPartAnimInstance|GetCurrentLOD (Variables|Default|GetBodyAccAnimInstance))))))

(event EventOnTargetingOrFirstPersonCameraChanged
  (bind _returnvalue (CustomCharacter|GetAssociatedPlayerPawn))
  ; ADS
  (|CallParentFunction)
  (Variables|Default|SetbIsADSing (Variables|/Script/FortniteGame.FortPlayerPawn|GetbIsTargeting _returnvalue))
  (if (Pawn|IsLocallyControlled _returnvalue)
    (Effects|Components|ParticleSystem|SetBoolParameter (CallFunction|GetIdleVFXComponent) "bIsADS" (Variables|Default|GetbIsADSing))))

(fn ConstructionScript
  (|CallParentFunction))
