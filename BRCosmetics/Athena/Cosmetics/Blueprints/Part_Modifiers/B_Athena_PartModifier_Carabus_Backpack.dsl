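; Cosmetic part modifier for the Carabus backpack that configures its idle hologram Niagara effect.
; After resources initialize, it feeds a dynamic hologram material into the effect, applies backpack-slot variants and attaches it to the part's mesh socket.
; User.IsDiving runs inverted to its name: 1 at setup and skydive end, 0 shortly after skydiving begins.
; An Endless-set character check is defined but not called here.
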(event EventOnPostResIn
  (|CallParentFunction)
  (CallFunction|SetupTexture))

(event Custom|Setup%20Texture
  (bind _returnvalue (CallFunction|GetIdleVFXComponent))
  ;[1 1 1] First we need to make sure that we actually have a dynamic material...
  (bind _returnvalue_2 (Rendering|Material|CreateDynamicMaterialInstance "/BRCosmetics/Accessories/FORT_Backpacks/Backpack_M_MED_Carabus/FX/M_Carabus_Holograms_Inst.M_Carabus_Holograms_Inst" "Elastic Hologram MID" "None" :!class /Script/Engine.KismetMaterialLibrary))
  (Niagara|SetNiagaraVariable(Material) (Utilities|Casting|CastToNiagaraComponent _returnvalue) "MID_CarabusHologram" _returnvalue_2)
  (Effects|Components|ParticleSystem|SetFloatParameter _returnvalue "User.IsDiving" 1.0)
  ;[1 1 1] Now, we can have the backpack handle applying the variant.
  (Class|FortCosmeticFunctionLibrary|ApplyVariantstoComponentbySlotTag (CustomCharacter|GetAssociatedPlayerPawn) "(TagName=\"LoadoutSlot.Backpack\")" (CallFunction|GetIdleVFXComponent)
    (:True
      _)
    (:False
      _))
  (Transformation|AttachComponentToComponent (CallFunction|GetIdleVFXComponent) (CustomCharacter|GetSkeletalMeshForPartOnPlayerPawn) (CallFunction|GetIdleFXSocketName) "SnapToTarget" "SnapToTarget" "SnapToTarget" true))

(event EventOnBeginSkydiving
  (|CallParentFunction)
  (Utilities|FlowControl|Delay 0.2)
  (Effects|Components|ParticleSystem|SetFloatParameter (CallFunction|GetIdleVFXComponent) "User.IsDiving" 0.0))

(event EventOnEndSkydiving
  (|CallParentFunction)
  (Effects|Components|ParticleSystem|SetFloatParameter (CallFunction|GetIdleVFXComponent) "User.IsDiving" 1.0))

(fn GrabLoadout
  (return (Variables|/Script/FortniteGame.FortPlayerPawn|GetCosmeticLoadout (CustomCharacter|GetAssociatedPlayerPawn))))

(fn IsCharacterElastic
  (return (GameplayTags|HasTag (GameplayTags|GetOwnedGameplayTags (CosmeticLoadout|GetItemDefinitionfromLoadout (CosmeticLoadout|GetCosmeticLoadout) "(TagName=\"LoadoutSlot.Character\")") :!class /Script/GameplayTags.GameplayTagAssetInterface) "(TagName=\"Cosmetics.Set.Endless\")")))

(fn ConstructionScript
  (|CallParentFunction))
