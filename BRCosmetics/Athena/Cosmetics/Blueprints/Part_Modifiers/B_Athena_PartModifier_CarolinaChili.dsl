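; Cosmetic part modifier for CarolinaChili that drives the "PlayerMoving" parameter on the idle VFX particle component.
; After post-res init, the frontend path uses delays and a FlipFlop to set that parameter for display without gameplay movement.
; Outside the frontend, init caches the associated Athena player pawn, and Tick sets the parameter when that pawn is sliding or tactical sprinting.
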(event EventOnPostResIn
  ;[0 0 0] post res
  (|CallParentFunction)
  (if (Reactive|Frontend|IsCosmeticObjectDisplayedinFrontendContext 0)
    (Utilities|FlowControl|Delay 5.0)
    (bind _isa (Utilities|FlowControl|FlipFlop)
      (:A
        _)
      (:B
        _))
    (if _isa
      (Utilities|FlowControl|RetriggerableDelay 5.0)
      (Effects|Components|ParticleSystem|SetBoolParameter (CallFunction|GetIdleVFXComponent) "PlayerMoving" _isa)
      (Utilities|FlowControl|RetriggerableDelay 2.0))
    (else
      (bind _asfort_player_pawn_athena (Utilities|Casting|CastToFortPlayerPawnAthena (CustomCharacter|GetAssociatedPlayerPawn)))
      (Variables|Default|SetplayerRef _asfort_player_pawn_athena))))

(event EventTick
  (params (DeltaSeconds :type Float%20%28single-precision%29))
  ;[0 0 0] tick fx
  (if (not (Reactive|Frontend|IsCosmeticObjectDisplayedinFrontendContext 0))
    (Effects|Components|ParticleSystem|SetBoolParameter (CallFunction|GetIdleVFXComponent) "PlayerMoving" (or (Variables|/Script/FortniteGame.FortPawn|GetbIsSliding (Variables|Default|GetplayerRef)) (Variables|/Script/FortniteGame.FortPlayerPawn|GetbIsTacticalSprinting (Variables|Default|GetplayerRef))))))

(fn ConstructionScript
  (|CallParentFunction))
