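; Character part modifier for the AnvilAlarmMoth cosmetic that drives a fake audio-reactive loop in frontend or preview displays.
; After post-res-in, a timeline ramps the EnableFrontendFakeAudioLoop parameter on the part's skeletal mesh materials and idle VFX particle component.
; On finish it waits for the configured fake amplitude on or off length, then reverses direction, repeating the pulse indefinitely.
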(event EventOnPostResIn
  (|CallParentFunction)
  (if (or (Reactive|Frontend|IsCosmeticObjectDisplayedinFrontendContext 0) (Reactive|Frontend|IsCosmeticObjectDisplayedasPreview 0))
    (bind (_direction _in) (Timeline|Timeline Play)
      (:Update
        (Rendering|Material|SetScalarParameterValueonMaterials (CustomCharacter|GetSkeletalMeshForPartOnPlayerPawn) "EnableFrontendFakeAudioLoop" _in) ; Enable Fake Audio in Frontend for Materials
        (Effects|Components|ParticleSystem|SetFloatParameter (CallFunction|GetIdleVFXComponent) "EnableFrontendFakeAudioLoop" _in))
      (:Finished
        (switch Utilities|FlowControl|Switch|SwitchonETimelineDirection _direction
          (:Forward
            (Utilities|FlowControl|Delay (Variables|AudioReactivity|GetAudioReactivityFakeAmplitudeOnLength))
            (Timeline|Timeline ReverseFromEnd))
          (:Backward
            (Utilities|FlowControl|Delay (Variables|AudioReactivity|GetAudioReactivityFakeAmplitudeOffLength))
            (Timeline|Timeline Play)))))))

(fn ConstructionScript
  (|CallParentFunction))
