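; Character part modifier for the BurnLeer cosmetic that drives kill-reactive fire materials and VFX.
; Reactivity writes timestamp and kill count to part materials, then a timeline animates FireReactive.
; Stat changes run another timeline feeding IsReacting to idle VFX and targeting state to elimination particles.
; Tick resets idle Niagara with Change_OffPhysics whenever a skin weight profile is active; the frontend reactivity branch is empty.
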(event EventTick
  (params (DeltaSeconds :type Float%20%28single-precision%29))
  ;[1 0.8011 0 0.5] SKIN WEIGTH FIX
  (if (Components|SkinnedMesh|IsUsingSkinWeightProfile (Customization|GetSkeletalMeshforPartType (CustomCharacter|GetAssociatedPlayerPawn) (CustomCharacter|GetCharacterPartType) :!class /Script/FortniteGame.FortPlayerPawn))
    (bind _asniagara_particle_system_component (Utilities|Casting|CastToNiagaraComponent (CallFunction|GetVFXComponent "(TagName=\"CosmeticPart.VFX.Idle\")")))
    (Niagara|SetNiagaraVariable(Bool) _asniagara_particle_system_component "Change_OffPhysics" true)
    (Niagara|ResetSystem _asniagara_particle_system_component)
    (Components|Activation|SetActive _asniagara_particle_system_component true true)))

(event EventOnPostResIn
  ;[0 0.564 1 0.5 30] FRONTEND 🛒
  (|CallParentFunction)
  (if (CallFunction|ShouldDisplayFrontendReactivity)
    _))

(event EventOnReactivity
  (params (KillCount :type Integer))
  (bind _returnvalue (CustomCharacter|GetSkeletalMeshforPartonOwner))
  ;[0.15 1 0.15 0.5] Material Reactivity
  (|CallParentFunction KillCount)
  (Rendering|Material|SetScalarParameterValueonMaterials _returnvalue "Timestamp" (Utilities|Time|GetGameTimeinSeconds))
  (Rendering|Material|SetScalarParameterValueonMaterials _returnvalue "KillCount" (Math|Conversions|ToFloat(Integer) KillCount))
  (bind (_direction _animationa) (Timeline|Timeline_0 PlayFromStart)
    (:Update
      (Rendering|Material|SetScalarParameterValueonMaterials (CustomCharacter|GetSkeletalMeshForPartOnPlayerPawn) "FireReactive" _animationa))))

(event EventOnStatChanged
  (params (State%20Name :type Name :default "None") (Stat%20Value :type Integer))
  (bind _eliminationparticle (Variables|Default|GetEliminationParticle))
  (|CallParentFunction State%20Name Stat%20Value)
  (bind (_direction _isreacting) (Timeline|Timeline PlayFromStart)
    (:Update
      (Effects|Components|ParticleSystem|SetBoolParameter _eliminationparticle "bIsAds" (Variables|Default|GetbLocalIsTargeting))
      (Effects|Components|ParticleSystem|SetFloatParameter (CallFunction|GetVFXComponent "(TagName=\"CosmeticPart.VFX.Idle\")") "IsReacting" _isreacting))
    (:Finished
      (Effects|Components|ParticleSystem|SetBoolParameter _eliminationparticle "bIsAds" false))))

(fn ConstructionScript
  (|CallParentFunction))
