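; Character part modifier for the IvoryDwell cosmetic, adding kill-reactive head aurora and shoulder effects.
; After res-in it attaches shoulders, arms and legs to the body mesh, then either plays a delayed frontend preview or watches the kill stat.
; Timelines drive Niagara reactivity and the shoulder material mask, with further kills during playback jumping the timeline instead of restarting.
; Outside the lobby, arm and leg visibility follows hidden hand and foot checks.
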(event EventOnPostResIn
  (bind _returnvalue (CustomCharacter|GetCharacterPartType))
  (bind _shoulders (Variables|Default|GetShoulders))
  (bind _ns_headaurora (Variables|Default|GetNS_HeadAurora))
  (|CallParentFunction)
  (bind _asfort_player_pawn (Utilities|Casting|CastToFortPlayerPawn (Actor|GetOwner)))
  (Transformation|AttachComponentToComponent _shoulders (Customization|GetSkeletalMeshforPartType _asfort_player_pawn "Body" :!class /Script/FortniteGame.FortPlayerPawn) "None" "SnapToTarget" "KeepRelative" "KeepRelative" true)
  (Transformation|AttachComponentToComponent (Variables|Default|GetArms) (Customization|GetSkeletalMeshforPartType _asfort_player_pawn "Body" :!class /Script/FortniteGame.FortPlayerPawn) "None" "SnapToTarget" "KeepRelative" "KeepRelative" true)
  (Transformation|AttachComponentToComponent (Variables|Default|GetLegs) (Customization|GetSkeletalMeshforPartType _asfort_player_pawn "Body" :!class /Script/FortniteGame.FortPlayerPawn) "None" "SnapToTarget" "KeepRelative" "KeepRelative" true)
  (Components|SkinnedMesh|SetLeaderPoseComponent _shoulders (Customization|GetSkeletalMeshforPartType _asfort_player_pawn "Body" :!class /Script/FortniteGame.FortPlayerPawn) true false)
  (bind _asniagara_particle_system_component (Utilities|Casting|CastToNiagaraComponent (CallFunction|GetIdleVFXComponent)))
  (Variables|Default|SetNS_HeadAurora _asniagara_particle_system_component)
  (if (Reactive|Frontend|ShouldDisplayFrontendReactivity 0 "Body")
    (Utilities|FlowControl|Delay 2.0)
    (Rendering|SetVisibility (Variables|Default|GetShoulders) true false)
    (bind (_direction _reactivity) (Timeline|TL_FrontEnd PlayFromStart)
      (:Update
        (switch Utilities|FlowControl|Switch|SwitchonEFortCustomPartType _returnvalue
          (:Body
            (Niagara|SetNiagaraVariable(Float) (Variables|Default|GetNS_HeadAurora) "Reactivity" _reactivity)
            (Rendering|Material|SetScalarParameterValueonMaterials (Variables|Default|GetShoulders) "FinalMaskMultiply" (* _reactivity 0.5)))))
      (:Finished
        (Rendering|SetVisibility (Variables|Default|GetShoulders) false false)))
    (else
      (Components|Activation|Deactivate (Variables|Default|GetNS_HeadAurora))
      (Cosmetics|WatchCosmeticStat (CustomCharacter|GetAssociatedPlayerPawn) "(TagName=\"StatManager.AthenaKills\")" 0 false
        (:OnStatChanged
          (if (and (Components|Timeline|IsPlaying (Variables|B_Athena_PartModifier_Character_IvoryDwell|GetTL)) (>= (Components|Timeline|GetPlaybackPosition (Variables|B_Athena_PartModifier_Character_IvoryDwell|GetTL)) 0.2))
            (CallFunction|RetriggerReactivity)
            (else
              (Niagara|ReinitializeSystem _ns_headaurora)
              (CallFunction|ChangeVFXVisibility)
              (Rendering|SetVisibility (Variables|Default|GetShoulders) true false)
              (bind (_direction_2 _reactivity_2) (Timeline|TL PlayFromStart)
                (:Update
                  (switch Utilities|FlowControl|Switch|SwitchonEFortCustomPartType _returnvalue
                    (:Body
                      (Niagara|SetNiagaraVariable(Float) (Variables|Default|GetNS_HeadAurora) "Reactivity" _reactivity_2)
                      (Rendering|Material|SetScalarParameterValueonMaterials (Variables|Default|GetShoulders) "FinalMaskMultiply" _reactivity_2))))
                (:Finished
                  (Components|Activation|Deactivate _ns_headaurora)
                  (Rendering|SetVisibility (Variables|Default|GetShoulders) false false))))))))))

(event EventOnWeaponChanged
  (params (NewWeapon :type Fort%20Weapon%20Object%20Reference) (OldWeapon :type Fort%20Weapon%20Object%20Reference))
  (|CallParentFunction NewWeapon OldWeapon)
  (CallFunction|ChangeVFXVisibility))

(event Custom|Retrigger%20Reactivity
  (Components|Timeline|SetPlaybackPosition (Variables|B_Athena_PartModifier_Character_IvoryDwell|GetTL) 0.75 false true))

(event EventOnCosmeticPartAppliedPostResIn
  (params (Part :type Custom%20Character%20Part%20Object%20Reference))
  (|CallParentFunction Part)
  (CallFunction|ChangeVFXVisibility))

(fn ChangeVFXVisibility
  (if (not (or (Reactive|Frontend|ShouldDisplayFrontendReactivity 0 "Body") (Reactive|Frontend|IsCosmeticObjectDisplayedinLobby 0)))
    (Rendering|SetVisibility (Variables|Default|GetArms) (not (Variables|HiddenPartCheck|GetbIsHandHidden)) false)
    (Rendering|SetVisibility (Variables|Default|GetLegs) (not (Variables|HiddenPartCheck|GetbIsFootHidden)) false)))

(fn ConstructionScript
  (|CallParentFunction)
  (Rendering|Material|SetScalarParameterValueonMaterials (Variables|Default|GetShoulders) "FinalMaskMultiply" 0.0)
  (Niagara|SetNiagaraVariable(Float) (Variables|Default|GetNS_HeadAurora) "Reactivity" 0.0))
