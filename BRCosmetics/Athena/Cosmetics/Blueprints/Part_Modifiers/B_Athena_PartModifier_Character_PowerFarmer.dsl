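; Character part modifier for the Power Farmer cosmetic that toggles its idle VFX according to the equipped material variant.
; After res-in it runs the parent logic, then queries the pawn's character loadout slot for the active Material channel index among Mat1 and Mat2.
; An index above zero, which corresponds to Mat2, activates the idle VFX component; any other result deactivates it.
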(event EventOnPostResIn
  (bind _returnvalue (CallFunction|GetIdleVFXComponent))
  (|CallParentFunction)
  (bind _index (CosmeticLoadout|GetVariantChannelActiveIndexforSlotActor "(TagName=\"LoadoutSlot.Character\")" (CustomCharacter|GetAssociatedPlayerPawn) "(TagName=\"Cosmetics.Variant.Channel.Material\")" (Utilities|Array|MakeArray "(TagName=\"Cosmetics.Variant.Property.Mat1\")" "(TagName=\"Cosmetics.Variant.Property.Mat2\")")))
  (if (> _index 0)
    (Components|Activation|Activate _returnvalue false)
    (else
      (Components|Activation|Deactivate _returnvalue))))

(fn ConstructionScript
  (|CallParentFunction))
