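; Character part modifier for the Sheerwill cosmetic that makes its materials and idle VFX react to eliminations through four stages.
; In matches it watches the AthenaKills stat to trigger each stage; in the front end it previews them on timed delays, then resets.
; Timelines animate per-stage material scalars, with some also driving particle activation and smoke progress.
; Post-update tick feeds the spine_05 world position to the material's OrbLocation.
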(event EventOnPostResIn
  (bind _returnvalue (CallFunction|GetIdleVFXComponent))
  (bind _returnvalue_2 (CallFunction|GetIdleVFXComponent))
  (|CallParentFunction)
  (if (CustomCosmeticModifier|IsinFrontEnd)
    (Effects|Components|ParticleSystem|SetFloatParameter _returnvalue "FX activate" 0.0)
    (Effects|Components|ParticleSystem|SetFloatParameter _returnvalue "FX Smoke Progress" 0.0)
    (Utilities|FlowControl|Delay 3.0)
    (CallFunction|ResetCheerwill)
    (CallFunction|Stage01)
    (Utilities|FlowControl|Delay 8.0)
    (CallFunction|Stage12)
    (Utilities|FlowControl|Delay 8.0)
    (CallFunction|Stage23)
    (Utilities|FlowControl|Delay 8.0)
    (CallFunction|Stage34)
    (Utilities|FlowControl|Delay 8.0)
    (CallFunction|ResetCheerwill)
    (else
      (Effects|Components|ParticleSystem|SetFloatParameter (CallFunction|GetIdleVFXComponent) "FX activate" 0.0)
      ;[0 0 0] Material Reactivity
      (bind (_statname _statvalue) (Cosmetics|WatchCosmeticStat (CustomCharacter|GetAssociatedPlayerPawn) "(TagName=\"StatManager.AthenaKills\")")
        (:OnStatChanged
          (Utilities|FlowControl|Sequence
            (:then_0
              (if (== _statvalue 1)
                (CallFunction|Stage01)
                (elif (== _statvalue 2)
                  (CallFunction|Stage12)
                  (elif (== _statvalue 3)
                    (CallFunction|Stage23)
                    (elif (== _statvalue 4)
                      (CallFunction|Stage34))))))
            (:then_1
              (if (> _statvalue 1)
                (Effects|Components|ParticleSystem|SetFloatParameter _returnvalue_2 "FX activate" 1.0)
                (Effects|Components|ParticleSystem|SetFloatParameter _returnvalue_2 "FX Smoke Progress" 0.0)))))))))

(event Custom|Stage0_1
  (bind _returnvalue (CustomCharacter|GetSkeletalMeshForPartOnPlayerPawn))
  ;[0 0 0] Stage0_1
  (bind (_direction _animationa) (Timeline|Timeline_0 PlayFromStart)
    (:Update
      (Rendering|Material|SetScalarParameterValueonMaterials _returnvalue "Stage0_1" _animationa))
    (:Finished
      (Rendering|Material|SetScalarParameterValueonMaterials _returnvalue "Stage0_1" 1.0))))

(event Custom|Stage1_2
  (bind _returnvalue (CustomCharacter|GetSkeletalMeshForPartOnPlayerPawn))
  ;[0 0 0 0.5] Stage1_2
  (bind (_direction _animationa _fxtrigged) (Timeline|Timeline_2 PlayFromStart)
    (:Update
      (Rendering|Material|SetScalarParameterValueonMaterials _returnvalue "Stage1_2" _animationa)
      (Effects|Components|ParticleSystem|SetFloatParameter (CallFunction|GetIdleVFXComponent) "FX activate" _fxtrigged))
    (:Finished
      (Rendering|Material|SetScalarParameterValueonMaterials _returnvalue "Stage1_2" 1.0))))

(event Custom|Stage2_3
  (bind _returnvalue (CustomCharacter|GetSkeletalMeshForPartOnPlayerPawn))
  ;[0 0 0 0.5] Stage2_3
  (bind (_direction _animationa _fxprogress) (Timeline|Timeline_3 PlayFromStart)
    (:Update
      (Rendering|Material|SetScalarParameterValueonMaterials _returnvalue "Stage2_3" _animationa)
      (Effects|Components|ParticleSystem|SetFloatParameter (CallFunction|GetIdleVFXComponent) "FX Smoke Progress" _fxprogress))
    (:Finished
      (Rendering|Material|SetScalarParameterValueonMaterials _returnvalue "Stage2_3" 1.0))))

(event Custom|Stage3_4
  (bind _returnvalue (CustomCharacter|GetSkeletalMeshForPartOnPlayerPawn))
  ;[0 0 0 0.5] Stage3_4
  (bind (_direction _animationa) (Timeline|Timeline_4 PlayFromStart)
    (:Update
      (Rendering|Material|SetScalarParameterValueonMaterials _returnvalue "Stage3_4" _animationa))
    (:Finished
      (Rendering|Material|SetScalarParameterValueonMaterials _returnvalue "Stage3_4" 1.0)))
  (Effects|Components|ParticleSystem|SetFloatParameter (CallFunction|GetIdleVFXComponent) "FX Smoke Progress" 1.0))

(event Custom|ResetCheerwill
  (bind _returnvalue (CustomCharacter|GetSkeletalMeshForPartOnPlayerPawn))
  ;[0 0 0 0.5] ResetCheerwill
  (Rendering|Material|SetScalarParameterValueonMaterials _returnvalue "Stage2_3" 0.0)
  (Rendering|Material|SetScalarParameterValueonMaterials _returnvalue "Stage3_4" 0.0)
  (Rendering|Material|SetScalarParameterValueonMaterials _returnvalue "Stage0_1" 0.0)
  (Rendering|Material|SetScalarParameterValueonMaterials _returnvalue "Stage1_2" 0.0))

(event EventBeginPlay
  ;[0 0 0] Comment
  (|CallParentFunction)
  (Actor|Tick|SetTickGroup "TG_PostUpdateWork"))

(event EventTick
  (params (DeltaSeconds :type Float%20%28single-precision%29))
  ;[0 0 0] Comment
  (|CallParentFunction DeltaSeconds)
  (Rendering|Material|SetVectorParameterValue (Class|BTechArtGlobalLibrary|GetDynamicMaterialInstance (CustomCharacter|GetSkeletalMeshForPartOnPlayerPawn)) "OrbLocation" (Math|Conversions|ToLinearColor(Vector) (.location (Transformation|GetSocketTransform (CustomCharacter|GetSkeletalMeshForPartOnPlayerPawn) "spine_05" "RTS_World"))) :!class /Script/Engine.MaterialInstanceDynamic))

(fn ConstructionScript
  (|CallParentFunction))
