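; Character part modifier for the Specter Dunk cosmetic that makes its idle VFX react to the knocked-down (DBNO) state.
; After res-in it binds to the associated player pawn's DBNO state change event.
; On each change it waits half a second, then sets the IsInDBNO bool parameter on the idle VFX particle component.
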(event EventOnPostResIn
  (|CallParentFunction)
  ; DBNO Reactivity
  (EventDispatchers|BindEventtoOnDBNOStateChanged (CustomCharacter|GetAssociatedPlayerPawn) (delegate Custom|OnDBNOStateChanged)))

(event Custom|OnDBNOStateChanged
  (params (FortPawn :type Fort%20Pawn%20Object%20Reference) (bInIsDBNO :type Boolean))
  ; DBNO Reactivity
  (Utilities|FlowControl|Delay 0.5)
  (Effects|Components|ParticleSystem|SetBoolParameter (CallFunction|GetIdleVFXComponent) "IsInDBNO" bInIsDBNO))

(fn ConstructionScript
  (|CallParentFunction))
