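; Character part modifier for the VigorKnackGolf cosmetic that reacts to eliminations and Victory Royales.
; Stat changes trigger kill reactivity, while a SeasonWins watcher fires on wins above the initial count; both play one timeline driving idle VFX and body material intensity.
; When frontend reactivity displays, it fakes kills on a delayed loop instead of watching wins.
; A repeating timer passes the part's anim-instance LOD to the idle VFX.
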(event EventOnPostResIn
  (bind _returnvalue (CustomCharacter|GetAssociatedPlayerPawn))
  ;[0.15 0.04315 0.05506 0.5] Reactivity Setup
  (|CallParentFunction)
  (Utilities|FlowControl|Sequence
    (:then_0
      (if (CallFunction|ShouldDisplayFrontendReactivity)
        (Variables|Default|SetFEKillCount 0)
        (Utilities|FlowControl|Delay 1.0)
        ;[0.15 0.04315 0.05506 0.5] FE Reactivity Loop
        (bind _result (Math|Integer|IncrementInt (Variables|Default|GetFEKillCount))
          (:%20%20
            (CallFunction|KillReactivity _result)
            (Utilities|FlowControl|Delay 6.5)))
        (else
          (bind _returnvalue_2 (Class|FortCosmeticFunctionLibrary|GetCosmeticStatValue _returnvalue "(TagName=\"StatManager.SeasonWins\")"))
          (Variables|Default|SetInitialWinCount _returnvalue_2)
          (bind (_statname _statvalue) (Cosmetics|WatchCosmeticStat _returnvalue "(TagName=\"StatManager.SeasonWins\")" 0 false)
            (:OnStatChanged
              ;[0.7478 0.7177 0.2899 0.5] Reactivity
              (CallFunction|VictoryRoyaleReactivity _statvalue))))))
    (:then_1
      (Utilities|Time|SetTimerbyEvent (EventDispatchers|CreateEvent :!event Custom|LODCheck) 0.2 true)
      (bind _ascustom_character_part_anim_instance (Utilities|Casting|CastToCustomCharacterPartAnimInstance (Components|SkeletalMesh|GetAnimInstance (Customization|GetSkeletalMeshforPartType (CustomCharacter|GetAssociatedPlayerPawn) (CustomCharacter|GetCharacterPartType) :!class /Script/FortniteGame.FortPlayerPawn))))
      (Variables|Default|SetAnimInstance _ascustom_character_part_anim_instance))))

(event Custom|LODCheck
  ;[0.8522 0.3853 0.163 0.5] LOD Check for Bones
  (Effects|Components|ParticleSystem|SetIntParameter (CallFunction|GetIdleVFXComponent) "CurrentLOD" (Variables|/Script/FortniteGame.CustomCharacterPartAnimInstance|GetCurrentLOD (Variables|Default|GetAnimInstance))))

(event EventOnStatChanged
  (params (State%20Name :type Name :default "None") (Stat%20Value :type Integer))
  ;[0.7478 0.7177 0.2899 0.5] Reactivity
  (|CallParentFunction State%20Name Stat%20Value)
  (CallFunction|KillReactivity Stat%20Value))

(event Custom|KillReactivity
  (params (KillCount :type Integer))
  ;[0.208 0.4 0.3005 0.5] FX/TA Elim/VictoryRoyale Reactivity
  (if (> KillCount 0)
    (bind (_direction _intensity) (Timeline|ReactivityTimeline PlayFromStart)
      (:Update
        (Effects|Components|ParticleSystem|SetFloatParameter (CallFunction|GetIdleVFXComponent) "ReactivityIntensity" _intensity)
        (Rendering|Material|SetScalarParameterValueonMaterials (Customization|GetSkeletalMeshforPartType (CustomCharacter|GetAssociatedPlayerPawn) "Body" :!class /Script/FortniteGame.FortPlayerPawn) "Reactivity Intensity" _intensity)))))

(event Custom|VictoryRoyaleReactivity
  (params (NewWinCount :type Integer))
  ;[0.208 0.4 0.3005 0.5] FX/TA Elim/VictoryRoyale Reactivity
  (if (and (> NewWinCount 0) (> NewWinCount (Variables|Default|GetInitialWinCount)))
    (Timeline|ReactivityTimeline PlayFromStart)))

(fn ConstructionScript
  (|CallParentFunction))
