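; Cosmetic part modifier for the Chicken Warrior character that feeds pawn movement state into its idle VFX component.
; On each movement update it sets the effect's on-ground flag and a running speed taken from the associated pawn's velocity length.
; Landing pulses a "Landed" flag that a retriggerable delay clears shortly after.
; The post-res-in event zeroes a separate "Land" float parameter.
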(event EventOnPostResIn
  (|CallParentFunction)
  (Effects|Components|ParticleSystem|SetFloatParameter (CallFunction|GetIdleVFXComponent) "Land" 0.0))

(event EventOnPawnLanded
  (bind _returnvalue (CallFunction|GetIdleVFXComponent))
  (|CallParentFunction)
  (Effects|Components|ParticleSystem|SetBoolParameter _returnvalue "Landed" true)
  (Utilities|FlowControl|RetriggerableDelay 0.34)
  (Effects|Components|ParticleSystem|SetBoolParameter _returnvalue "Landed" false))

(event EventOnCharacterMovementUpdated
  (params (DeltaSeconds :type Float%20%28single-precision%29) (OldLocation :type Vector :default "0, 0, 0") (OldVelocity :type Vector :default "0, 0, 0"))
  (bind _returnvalue (CallFunction|GetIdleVFXComponent))
  (bind _returnvalue_2 (CustomCharacter|GetAssociatedPlayerPawn))
  (bind _returnvalue_3 (Transformation|GetVelocity _returnvalue_2))
  (|CallParentFunction DeltaSeconds OldLocation OldVelocity)
  (Effects|Components|ParticleSystem|SetBoolParameter (CallFunction|GetIdleVFXComponent) "IsOnGround" (AI|Components|NavMovement|IsMovingonGround (Variables|/Script/Engine.Character|GetCharacterMovement _returnvalue_2)))
  (if (Math|Vector|Equal(Vector) _returnvalue_3 "0, 0, 0" 0.0001)
    (Effects|Components|ParticleSystem|SetFloatParameter _returnvalue "RunningSpeed" 0.0)
    (else
      (Effects|Components|ParticleSystem|SetFloatParameter _returnvalue "RunningSpeed" (Math|Vector|VectorLength _returnvalue_3)))))

(fn ConstructionScript)
