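; ChocoBunny cosmetic part modifier Blueprint that adds a particle effect to that skin's character part.
; When the pawn lands, it runs the parent handler, then spawns the NS_ChocoBunny Niagara system attached at the root of the part's skeletal mesh on the player pawn.
; The construction script only calls its parent.
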(event EventOnPawnLanded
  (|CallParentFunction)
  (PartModifier|SpawnSystem|SpawnSystemAttachedPartModifier "/BRCosmetics/Athena/Effects/Particles/Skins/ChocoBunny/NS_ChocoBunny.NS_ChocoBunny" (CustomCharacter|GetSkeletalMeshForPartOnPlayerPawn) "root"))

(fn ConstructionScript
  (|CallParentFunction))
