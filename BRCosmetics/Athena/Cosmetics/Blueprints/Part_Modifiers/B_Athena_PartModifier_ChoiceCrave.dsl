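; Character part modifier for the ChoiceCrave cosmetic that spawns an attached elimination VFX as a reactive effect.
; When any reported stat change has a value above zero, it spawns the tagged effect and passes the hand-hidden flag to Niagara.
; After res-in, if frontend reactivity should display, it spawns the same effect after a short delay, without setting that flag.
; The trailing delay on that path leads to no further action.
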(event EventOnPostResIn
  (|CallParentFunction)
  (if (CallFunction|ShouldDisplayFrontendReactivity)
    (Utilities|FlowControl|Delay 1.5)
    (bind _returnvalue (CallFunction|GetVFXSpawnParameters "(TagName=\"CosmeticPart.VFX.Elimination\")")
      (:Success
        (PartModifier|SpawnSystem|SpawnSystemAttachedPartModifier (Utilities|Struct|BreakCharacterPartModifierVFXSpawnParameters _returnvalue) (.AttachToComponent (Utilities|Struct|BreakCharacterPartModifierVFXSpawnParameters _returnvalue)) (.SocketName (Utilities|Struct|BreakCharacterPartModifierVFXSpawnParameters _returnvalue)) "0, 0, 0" "0, 0, 0" (.LocationType (Utilities|Struct|BreakCharacterPartModifierVFXSpawnParameters _returnvalue)) true (.bAutoActivate (Utilities|Struct|BreakCharacterPartModifierVFXSpawnParameters _returnvalue))
          (:Success
            (Utilities|FlowControl|Delay 1e+01)))))))

(event EventOnStatChanged
  (params (State%20Name :type Name :default "None") (Stat%20Value :type Integer))
  (|CallParentFunction State%20Name Stat%20Value)
  (if (> Stat%20Value 0)
    (bind _returnvalue (CallFunction|GetVFXSpawnParameters "(TagName=\"CosmeticPart.VFX.Elimination\")")
      (:Success
        (bind _returnvalue_2 (PartModifier|SpawnSystem|SpawnSystemAttachedPartModifier (Utilities|Struct|BreakCharacterPartModifierVFXSpawnParameters _returnvalue) (.AttachToComponent (Utilities|Struct|BreakCharacterPartModifierVFXSpawnParameters _returnvalue)) (.SocketName (Utilities|Struct|BreakCharacterPartModifierVFXSpawnParameters _returnvalue)) "0, 0, 0" "0, 0, 0" (.LocationType (Utilities|Struct|BreakCharacterPartModifierVFXSpawnParameters _returnvalue)) true (.bAutoActivate (Utilities|Struct|BreakCharacterPartModifierVFXSpawnParameters _returnvalue)))
          (:Success
            (Niagara|SetNiagaraVariable(Bool) _returnvalue_2 "bisHandHidden" (Variables|HiddenPartCheck|GetbIsHandHidden))))))))

(fn ConstructionScript
  (|CallParentFunction))
