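; Cosmetic part modifier for CiderPetalClover that makes a character part visually react to eliminations.
; After post-res-in it enables the material reactivity curve and scales the timeline play rate to the reactivity duration.
; In the front end it plays the reaction then starts a delay; otherwise the AthenaKills stat watcher triggers it.
; The timeline drives the mesh material's ReactivityCurve and the idle VFX's FXColorSwitch.
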(event EventOnPostResIn
  (|CallParentFunction)
  (Rendering|Material|SetScalarParameterValueonMaterials (CustomCharacter|GetSkeletalMeshforPartonOwner) "UseReactivityCurve" 1.0)
  (Components|Timeline|SetPlayRate (Variables|B_Athena_PartModifier_CiderPetalClover|GetReactivity_Timeline) (/ 1.0 (Variables|Default|GetReactivityDuration)))
  (if (CustomCosmeticModifier|IsinFrontEnd)
    (CallFunction|OnReactivity)
    (Utilities|FlowControl|Delay (+ (Variables|Default|GetFE_LoopWait) (Variables|Default|GetReactivityDuration)))
    (else
      (Cosmetics|WatchCosmeticStat (CustomCharacter|GetAssociatedPlayerPawn) "(TagName=\"StatManager.AthenaKills\")" 0 false
        (:OnStatChanged
          (CallFunction|OnReactivity))))))

(event Custom|OnReactivity
  (bind (_direction _reactivitycurve _fxcolorcurve) (Timeline|Reactivity_Timeline PlayFromStart)
    (:Update
      (Rendering|Material|SetScalarParameterValueonMaterials (CustomCharacter|GetSkeletalMeshforPartonOwner) "ReactivityCurve" _reactivitycurve)
      (Effects|Components|ParticleSystem|SetFloatParameter (CallFunction|GetIdleVFXComponent) "FXColorSwitch" _fxcolorcurve))))

(fn ConstructionScript
  (|CallParentFunction))
