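; Clash skin cosmetic part modifier that shows a reactive effect, apparently for kills, when a tracked stat changes to a positive value.
; If the effect is inactive and the Progressive variant index is 1, it spawns an attached system with hidden hand/foot flags and starts a timeline.
; The timeline animates spike and tendril parameters on the idle Niagara component.
; Further triggers during the effect jump the timeline to a set time instead of respawning.
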(event EventOnStatChanged
  (params (State%20Name :type Name :default "None") (Stat%20Value :type Integer))
  (|CallParentFunction State%20Name Stat%20Value)
  (if (> Stat%20Value 0)
    (bind _asniagara_particle_system_component (Utilities|Casting|CastToNiagaraComponent (CallFunction|GetIdleVFXComponent)))
    (if (Variables|Default|GetIs_Skin_Kill_FX_Visibility)
      (bind (_direction _spikes_number _big_meshtendrils_size _big_meshtendrils_location _big_meshtendrils_dissolve _small_meshtendrils_size _small_meshtendrils_location _big_mesh02tendrils_size) (Timeline|Timeline_0 SetNewTime :NewTime 0.35)
        (:Update
          (Utilities|FlowControl|Sequence
            (:then_1
              (Niagara|SetNiagaraVariable(Float) _asniagara_particle_system_component "Spikes_number" _spikes_number)
              (Niagara|SetNiagaraVariable(Float) _asniagara_particle_system_component "Big_MeshTendrils_Size" _big_meshtendrils_size)
              (Niagara|SetNiagaraVariable(Float) _asniagara_particle_system_component "Big_MeshTendrils_Location" _big_meshtendrils_location)
              (Niagara|SetNiagaraVariable(Float) _asniagara_particle_system_component "Big_MeshTendrils_Dissolve" _big_meshtendrils_dissolve)
              (Niagara|SetNiagaraVariable(Float) _asniagara_particle_system_component "Small_MeshTendrils_Size" _small_meshtendrils_size)
              (Niagara|SetNiagaraVariable(Float) _asniagara_particle_system_component "Small_MeshTendrils_Location" _small_meshtendrils_location)
              (Niagara|SetNiagaraVariable(Float) _asniagara_particle_system_component "Big_Mesh02Tendrils_Size" _big_mesh02tendrils_size))))
        (:Finished
          (Variables|Default|SetIs_Skin_Kill_FX_Visibility false)))
      (else
        (bind _index (CosmeticLoadout|GetVariantChannelActiveIndexforSlotActor "(TagName=\"LoadoutSlot.Character\")" (CustomCharacter|GetAssociatedPlayerPawn) "(TagName=\"Cosmetics.Variant.Channel.Progressive\")" (Utilities|Array|MakeArray "(TagName=\"Cosmetics.Variant.Property.Stage1\")" "(TagName=\"Cosmetics.Variant.Property.Stage2\")")))
        (bind _output_get (Variables|Default|SetCosmeticLoadoutInt _index))
        (if (== _output_get 1)
          (bind _returnvalue (PartModifier|SpawnSystem|SpawnSystemAttachedPartModifier 0 (CustomCharacter|GetSkeletalMeshForPartOnPlayerPawn) "root")
            (:Success
              (Effects|Components|ParticleSystem|SetBoolParameter _returnvalue "bisHandHidden" (Variables|HiddenPartCheck|GetbIsHandHidden))
              (Effects|Components|ParticleSystem|SetBoolParameter _returnvalue "bisFootHidden" (Variables|HiddenPartCheck|GetbIsFootHidden))
              (Variables|Default|SetIs_Skin_Kill_FX_Visibility true)
              (Timeline|Timeline_0 PlayFromStart))))))))

(fn ConstructionScript
  (|CallParentFunction))
