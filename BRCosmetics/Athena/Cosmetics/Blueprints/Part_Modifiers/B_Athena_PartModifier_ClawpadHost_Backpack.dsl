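; Cosmetic part modifier for the ClawpadHost backpack that briefly glitches its look from "Cody" to "66" and back.
; On post-init it sets custom depth and translucency sorting, then watches the Athena kills stat with a cooldown flag, or runs timed in the frontend.
; Delayed material scalar changes drive each swap, pulsing ReactiveTriggered on the idle VFX.
; A duplicate custom event repeats the sequence as the frontend loop.
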(event EventOnPostResIn
  (bind _returnvalue (CustomCharacter|GetSkeletalMeshForPartOnPlayerPawn))
  (bind _returnvalue_2 (CallFunction|GetIdleVFXComponent))
  (|CallParentFunction)
  (if (Reactive|Frontend|ShouldDisplayFrontendReactivity 0 "Backpack")
    (Utilities|FlowControl|Delay 0.15))
  (Rendering|SetRenderCustomDepth (Customization|GetSkeletalMeshforPartType (CustomCharacter|GetAssociatedPlayerPawn) "Backpack" :!class /Script/FortniteGame.FortPlayerPawn) true)
  (Rendering|SetTranslucencySortDistanceOffset _returnvalue 18.0)
  (if (Reactive|Frontend|ShouldDisplayFrontendReactivity 0 "Backpack")
    (Utilities|FlowControl|Delay 4.0)
    (else
      (bind (_statname _statvalue) (Cosmetics|WatchCosmeticStat (CustomCharacter|GetAssociatedPlayerPawn) "(TagName=\"StatManager.AthenaKills\")")
        (:OnStatChanged
          (if (> _statvalue 0)
            (if (Variables|Default|GetCooldown)
              (Variables|Default|SetCooldown false)
              (else
                (Variables|Default|SetCooldown false))))))))
  (Rendering|Material|SetScalarParameterValueonMaterials _returnvalue "Glitch Transition" 1.0)
  (Utilities|FlowControl|Delay (Variables|Default|GetGlitch%20Time))
  (Effects|Components|ParticleSystem|SetBoolParameter _returnvalue_2 "ReactiveTriggered" true)
  (Rendering|Material|SetScalarParameterValueonMaterials _returnvalue "Cody Visibility" 0.0)
  (Rendering|Material|SetScalarParameterValueonMaterials _returnvalue "66 Visibility" 1.0)
  (Utilities|FlowControl|Delay (Variables|Default|GetGlitch%20Time))
  (Rendering|Material|SetScalarParameterValueonMaterials _returnvalue "Glitch Transition" 0.0)
  (Effects|Components|ParticleSystem|SetBoolParameter _returnvalue_2 "ReactiveTriggered" false)
  (Utilities|FlowControl|Delay (Math|Float|SelectFloat (Variables|Default|GetReactiveShowTime_FE) (Variables|Default|GetReactiveShowTime_Game) (Reactive|Frontend|ShouldDisplayFrontendReactivity 0 "Backpack")))
  (Rendering|Material|SetScalarParameterValueonMaterials _returnvalue "Glitch Transition" 1.0)
  (Utilities|FlowControl|Delay (Variables|Default|GetGlitch%20Time))
  (Effects|Components|ParticleSystem|SetBoolParameter _returnvalue_2 "ReactiveTriggered" true)
  (Rendering|Material|SetScalarParameterValueonMaterials _returnvalue "66 Visibility" 0.0)
  (Rendering|Material|SetScalarParameterValueonMaterials _returnvalue "Cody Visibility" 1.0)
  (Utilities|FlowControl|Delay (Variables|Default|GetGlitch%20Time))
  (Effects|Components|ParticleSystem|SetBoolParameter _returnvalue_2 "ReactiveTriggered" false)
  (Rendering|Material|SetScalarParameterValueonMaterials _returnvalue "Glitch Transition" 0.0)
  (if (Reactive|Frontend|ShouldDisplayFrontendReactivity 0 "Backpack")
    (CallFunction|FrontendReactivityLoop)
    (else
      (Utilities|FlowControl|Delay (Variables|Default|GetCooldownTime))
      (Variables|Default|SetCooldown true))))

(event Custom|Frontend%20Reactivity%20Loop
  (bind _returnvalue (CustomCharacter|GetSkeletalMeshForPartOnPlayerPawn))
  (bind _returnvalue_2 (CallFunction|GetIdleVFXComponent))
  (Utilities|FlowControl|Delay 4.0)
  (Rendering|Material|SetScalarParameterValueonMaterials _returnvalue "Glitch Transition" 1.0)
  (Utilities|FlowControl|Delay (Variables|Default|GetGlitch%20Time))
  (Effects|Components|ParticleSystem|SetBoolParameter _returnvalue_2 "ReactiveTriggered" true)
  (Rendering|Material|SetScalarParameterValueonMaterials _returnvalue "Cody Visibility" 0.0)
  (Rendering|Material|SetScalarParameterValueonMaterials _returnvalue "66 Visibility" 1.0)
  (Utilities|FlowControl|Delay (Variables|Default|GetGlitch%20Time))
  (Rendering|Material|SetScalarParameterValueonMaterials _returnvalue "Glitch Transition" 0.0)
  (Effects|Components|ParticleSystem|SetBoolParameter _returnvalue_2 "ReactiveTriggered" false)
  (Utilities|FlowControl|Delay (Math|Float|SelectFloat (Variables|Default|GetReactiveShowTime_FE) (Variables|Default|GetReactiveShowTime_Game) (Reactive|Frontend|ShouldDisplayFrontendReactivity 0 "Backpack")))
  (Rendering|Material|SetScalarParameterValueonMaterials _returnvalue "Glitch Transition" 1.0)
  (Utilities|FlowControl|Delay (Variables|Default|GetGlitch%20Time))
  (Effects|Components|ParticleSystem|SetBoolParameter _returnvalue_2 "ReactiveTriggered" true)
  (Rendering|Material|SetScalarParameterValueonMaterials _returnvalue "66 Visibility" 0.0)
  (Rendering|Material|SetScalarParameterValueonMaterials _returnvalue "Cody Visibility" 1.0)
  (Utilities|FlowControl|Delay (Variables|Default|GetGlitch%20Time))
  (Effects|Components|ParticleSystem|SetBoolParameter _returnvalue_2 "ReactiveTriggered" false)
  (Rendering|Material|SetScalarParameterValueonMaterials _returnvalue "Glitch Transition" 0.0)
  (if (Reactive|Frontend|ShouldDisplayFrontendReactivity 0 "Backpack")
    (CallFunction|FrontendReactivityLoop)
    (else
      (Utilities|FlowControl|Delay (Variables|Default|GetCooldownTime))
      (Variables|Default|SetCooldown true))))

(fn ConstructionScript
  (|CallParentFunction))
