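; Character part modifier for the Cloaked Assassin body that makes the skin glow brighter across reactive stages.
; In matches it starts unlit, stat changes update the idle VFX "KillCount" and trigger Niagara bursts, and stage activations set emissive material scalars.
; In the front end, a timeline-driven loop cycles StateInt through the stages as a preview.
; The preview duplicates the stage values, except stage 3's max brightness is 7.0 versus 0.7.
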(event EventOnPostResIn
  (bind _returnvalue (CustomCharacter|GetSkeletalMeshForPartOnPlayerPawn))
  (bind _returnvalue_2 (CustomCharacter|GetSkeletalMeshForPartOnPlayerPawn))
  (|CallParentFunction)
  (bind _returnvalue_3 (Fortnite|GetDynamicMaterialInstance (Customization|GetSkeletalMeshforPartType (CustomCharacter|GetAssociatedPlayerPawn) (CustomCharacter|GetCharacterPartType) :!class /Script/FortniteGame.FortPlayerPawn)))
  (Variables|Default|SetCube%20Dyna%20Mat _returnvalue_3)
  (if (CustomCosmeticModifier|IsinFrontEnd)
    (Utilities|FlowControl|Delay 0.2)
    ;[1 1 1] Frontend Cycling
    (Utilities|FlowControl|DoOnce false
      (:Completed
        (switch int (Variables|Default|GetStateInt) :!label _label
          (:0
            (Rendering|Material|SetScalarParameterValueonMaterials _returnvalue_2 "Emissive_BrightnessMax" 0.0)
            (Rendering|Material|SetScalarParameterValueonMaterials _returnvalue_2 "Emissive_BrightnessMin" 0.0)
            (Rendering|Material|SetScalarParameterValueonMaterials _returnvalue_2 "Emissive mult" 0.0))
          (:1
            (Rendering|Material|SetScalarParameterValueonMaterials _returnvalue_2 "Emissive_BrightnessMax" 0.3)
            (Rendering|Material|SetScalarParameterValueonMaterials _returnvalue_2 "Emissive_BrightnessMin" 0.0)
            (Rendering|Material|SetScalarParameterValueonMaterials _returnvalue_2 "Emissive mult" 1e+01))
          (:2
            (Rendering|Material|SetScalarParameterValueonMaterials _returnvalue_2 "Emissive_BrightnessMax" 0.4)
            (Rendering|Material|SetScalarParameterValueonMaterials _returnvalue_2 "Emissive_BrightnessMin" 0.0)
            (Rendering|Material|SetScalarParameterValueonMaterials _returnvalue_2 "Emissive mult" 11.0))
          (:3
            (Rendering|Material|SetScalarParameterValueonMaterials _returnvalue_2 "Emissive_BrightnessMax" 7.0)
            (Rendering|Material|SetScalarParameterValueonMaterials _returnvalue_2 "Emissive_BrightnessMin" 0.0)
            (Rendering|Material|SetScalarParameterValueonMaterials _returnvalue_2 "Emissive mult" 14.0)))
        (Effects|Components|ParticleSystem|SetIntParameter (CallFunction|GetIdleVFXComponent) "KillCount" (Variables|Default|GetStateInt))
        (if (not (== (Variables|Default|GetStateInt) 0))
          (PartModifier|SpawnSystem|SpawnSystemAttachedPartModifier "/BRCosmetics/Characters/Player/Female/Medium/Bodies/F_MED_Cloaked_Assassin/Effects/NS_Skin_CloakedAssassin_Burst.NS_Skin_CloakedAssassin_Burst" (CustomCharacter|GetSkeletalMeshForPartOnPlayerPawn) "root"
            (:Success
              (if (< (Variables|Default|GetStateInt) 3)
                (Variables|Default|SetStateInt (+ (Variables|Default|GetStateInt) 1) :!label _label_2)
                (else
                  (Variables|Default|SetStateInt 0)))
              (Timeline|Timeline_0 PlayFromStart
                (:Finished
                  (goto _label))))))
        (goto _label_2)))
    (else
      (Utilities|FlowControl|Delay 1.0)
      ;[0.5136 1 0.92] Set parameters directly on the skel mesh of character parts
      (Rendering|Material|SetScalarParameterValueonMaterials _returnvalue "Emissive_BrightnessMax" 0.0)
      (Rendering|Material|SetScalarParameterValueonMaterials _returnvalue "Emissive_BrightnessMin" 0.0)
      (Rendering|Material|SetScalarParameterValueonMaterials _returnvalue "Emissive mult" 0.0))))

(event EventOnStatChanged
  (params (State%20Name :type Name :default "None") (Stat%20Value :type Integer))
  (|CallParentFunction State%20Name Stat%20Value)
  (Effects|Components|ParticleSystem|SetIntParameter (CallFunction|GetIdleVFXComponent) "KillCount" Stat%20Value)
  (Utilities|FlowControl|Delay 0.2)
  (if (> Stat%20Value 0)
    (PartModifier|SpawnSystem|SpawnSystemAttachedPartModifier "/BRCosmetics/Characters/Player/Female/Medium/Bodies/F_MED_Cloaked_Assassin/Effects/NS_Skin_CloakedAssassin_Burst.NS_Skin_CloakedAssassin_Burst" (CustomCharacter|GetSkeletalMeshForPartOnPlayerPawn) "root")))

(event EventOnCosmeticStateActivated
  (params (Stat%20Name :type Name :default "None") (Stat%20Value :type Integer) (STAGE%20Num%20Being%20Activated :type Integer))
  (bind _returnvalue (CustomCharacter|GetSkeletalMeshForPartOnPlayerPawn))
  (|CallParentFunction Stat%20Name Stat%20Value STAGE%20Num%20Being%20Activated)
  ;[1 1 1] Stage Materials set
  (switch int STAGE%20Num%20Being%20Activated
    (:0
      ;[0.602 1 0.9601] Set parameters directly on the skel mesh of character parts
      (Rendering|Material|SetScalarParameterValueonMaterials _returnvalue "Emissive_BrightnessMax" 0.0)
      (Rendering|Material|SetScalarParameterValueonMaterials _returnvalue "Emissive_BrightnessMin" 0.0)
      (Rendering|Material|SetScalarParameterValueonMaterials _returnvalue "Emissive mult" 0.0))
    (:1
      (Rendering|Material|SetScalarParameterValueonMaterials _returnvalue "Emissive_BrightnessMax" 0.3)
      (Rendering|Material|SetScalarParameterValueonMaterials _returnvalue "Emissive_BrightnessMin" 0.0)
      (Rendering|Material|SetScalarParameterValueonMaterials _returnvalue "Emissive mult" 1e+01))
    (:2
      (Rendering|Material|SetScalarParameterValueonMaterials _returnvalue "Emissive_BrightnessMax" 0.4)
      (Rendering|Material|SetScalarParameterValueonMaterials _returnvalue "Emissive_BrightnessMin" 0.0)
      (Rendering|Material|SetScalarParameterValueonMaterials _returnvalue "Emissive mult" 11.0))
    (:3
      (Rendering|Material|SetScalarParameterValueonMaterials _returnvalue "Emissive_BrightnessMax" 0.7)
      (Rendering|Material|SetScalarParameterValueonMaterials _returnvalue "Emissive_BrightnessMin" 0.0)
      (Rendering|Material|SetScalarParameterValueonMaterials _returnvalue "Emissive mult" 14.0))))

(fn ConstructionScript
  (|CallParentFunction))
