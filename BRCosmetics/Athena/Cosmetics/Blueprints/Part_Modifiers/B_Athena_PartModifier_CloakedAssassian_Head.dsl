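; Cosmetic part modifier for the Cloaked Assassin head that raises its emissive glow across four stages.
; On cosmetic state activation, the stage number selects brightness and multiplier scalars set directly on the part's skeletal mesh materials.
; In the front end, post-res-in starts a delayed, run-once loop that cycles all stages, using a timeline's finish to re-enter the switch.
; Post-res-in also caches a dynamic material instance, which these graphs never read.
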(event Custom|FE_Version
  (bind _returnvalue (CustomCharacter|GetSkeletalMeshForPartOnPlayerPawn))
  ;[1 1 1] Frontend Cycling
  (Utilities|FlowControl|Delay 0.2)
  (Utilities|FlowControl|DoOnce false
    (:Completed
      (switch int (Variables|Default|GetStateInt) :!label _label
        (:0
          ;[0.3228 1 0.9901] Set parameters directly on the skel mesh of character parts
          (Rendering|Material|SetScalarParameterValueonMaterials _returnvalue "Emissive_BrightnessMax" 0.03)
          (Rendering|Material|SetScalarParameterValueonMaterials _returnvalue "Emissive_BrightnessMin" 0.03)
          (Rendering|Material|SetScalarParameterValueonMaterials _returnvalue "Emissive mult" 0.0))
        (:1
          (Rendering|Material|SetScalarParameterValueonMaterials _returnvalue "Emissive_BrightnessMax" 0.1)
          (Rendering|Material|SetScalarParameterValueonMaterials _returnvalue "Emissive_BrightnessMin" 0.075)
          (Rendering|Material|SetScalarParameterValueonMaterials _returnvalue "Emissive mult" 5.0))
        (:2
          (Rendering|Material|SetScalarParameterValueonMaterials _returnvalue "Emissive_BrightnessMax" 0.35)
          (Rendering|Material|SetScalarParameterValueonMaterials _returnvalue "Emissive_BrightnessMin" 0.1)
          (Rendering|Material|SetScalarParameterValueonMaterials _returnvalue "Emissive mult" 8.0))
        (:3
          (Rendering|Material|SetScalarParameterValueonMaterials _returnvalue "Emissive_BrightnessMax" 0.7)
          (Rendering|Material|SetScalarParameterValueonMaterials _returnvalue "Emissive_BrightnessMin" 0.2)
          (Rendering|Material|SetScalarParameterValueonMaterials _returnvalue "Emissive mult" 1e+01)))
      (if (< (Variables|Default|GetStateInt) 3)
        (Variables|Default|SetStateInt (+ (Variables|Default|GetStateInt) 1))
        (else
          (Variables|Default|SetStateInt 0)))
      (Timeline|Timeline_0 PlayFromStart
        (:Finished
          (goto _label))))))

(event EventOnPostResIn
  (|CallParentFunction)
  (bind _returnvalue (Fortnite|GetDynamicMaterialInstance (Customization|GetSkeletalMeshforPartType (CustomCharacter|GetAssociatedPlayerPawn) (CustomCharacter|GetCharacterPartType) :!class /Script/FortniteGame.FortPlayerPawn)))
  (Variables|Default|SetCube%20Dyna%20Mat _returnvalue)
  (if (CustomCosmeticModifier|IsinFrontEnd)
    (CallFunction|FEVersion)))

(event EventOnCosmeticStateActivated
  (params (Stat%20Name :type Name :default "None") (Stat%20Value :type Integer) (STAGE%20Num%20Being%20Activated :type Integer))
  (bind _returnvalue (CustomCharacter|GetSkeletalMeshForPartOnPlayerPawn))
  (|CallParentFunction Stat%20Name Stat%20Value STAGE%20Num%20Being%20Activated)
  ;[1 1 1] Stage Materials set
  (switch int STAGE%20Num%20Being%20Activated
    (:0
      ;[0.3822 0.9402 1] Set parameters directly on the skel mesh of character parts
      (Rendering|Material|SetScalarParameterValueonMaterials _returnvalue "Emissive_BrightnessMax" 0.03)
      (Rendering|Material|SetScalarParameterValueonMaterials _returnvalue "Emissive_BrightnessMin" 0.03)
      (Rendering|Material|SetScalarParameterValueonMaterials _returnvalue "Emissive mult" 0.0))
    (:1
      (Rendering|Material|SetScalarParameterValueonMaterials _returnvalue "Emissive_BrightnessMax" 0.1)
      (Rendering|Material|SetScalarParameterValueonMaterials _returnvalue "Emissive_BrightnessMin" 0.075)
      (Rendering|Material|SetScalarParameterValueonMaterials _returnvalue "Emissive mult" 5.0))
    (:2
      (Rendering|Material|SetScalarParameterValueonMaterials _returnvalue "Emissive_BrightnessMax" 0.35)
      (Rendering|Material|SetScalarParameterValueonMaterials _returnvalue "Emissive_BrightnessMin" 0.1)
      (Rendering|Material|SetScalarParameterValueonMaterials _returnvalue "Emissive mult" 8.0))
    (:3
      (Rendering|Material|SetScalarParameterValueonMaterials _returnvalue "Emissive_BrightnessMax" 0.7)
      (Rendering|Material|SetScalarParameterValueonMaterials _returnvalue "Emissive_BrightnessMin" 0.2)
      (Rendering|Material|SetScalarParameterValueonMaterials _returnvalue "Emissive mult" 1e+01))))

(fn ConstructionScript
  (|CallParentFunction))
