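; Character part modifier for the Clyde SL DenimEquip cosmetic that feeds live skeleton positions to the part's materials.
; After res-in it sets BP_UseCCPM to 1 on the player pawn's part mesh materials.
; Each tick it pushes both hand positions, the chest position and per-hand offset vectors (metacarpal minus middle finger socket) as vector parameters.
; GameTimeFromBPForGlow only returns zero and is not called here.
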(event EventOnPostResIn
  (|CallParentFunction)
  (Rendering|Material|SetScalarParameterValueonMaterials (CustomCharacter|GetSkeletalMeshForPartOnPlayerPawn) "BP_UseCCPM" 1.0))

(event EventTick
  (params (DeltaSeconds :type Float%20%28single-precision%29))
  (bind _returnvalue (CustomCharacter|GetSkeletalMeshForPartOnPlayerPawn))
  (bind _returnvalue_2 (Transformation|GetSocketLocation _returnvalue "middle_metacarpal_l"))
  (bind _returnvalue_3 (Transformation|GetSocketLocation _returnvalue "middle_metacarpal_r"))
  (Rendering|Material|SetVectorParameterValueonMaterials _returnvalue "BP_HandPosition_L" _returnvalue_2)
  (Rendering|Material|SetVectorParameterValueonMaterials _returnvalue "BP_HandPosition_R" _returnvalue_3)
  (Rendering|Material|SetVectorParameterValueonMaterials _returnvalue "BP_ChestPosition" (Transformation|GetSocketLocation _returnvalue "spine_05"))
  (Rendering|Material|SetVectorParameterValueonMaterials _returnvalue "BP_OffsetVector_L" (- _returnvalue_2 (Transformation|GetSocketLocation _returnvalue "middle_01_l")))
  (Rendering|Material|SetVectorParameterValueonMaterials _returnvalue "BP_OffsetVector_R" (- _returnvalue_3 (Transformation|GetSocketLocation _returnvalue "middle_01_r"))))

(fn GameTimeFromBPForGlow
  (params (A :type Float%20%28double-precision%29) (B :type Float%20%28double-precision%29))
  (return 0.0))

(fn ConstructionScript)
