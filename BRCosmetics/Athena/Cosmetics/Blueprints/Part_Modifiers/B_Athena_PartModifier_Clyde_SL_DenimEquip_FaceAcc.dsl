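; Cosmetic part modifier for the Clyde SL DenimEquip face accessory that feeds live bone positions to the character part's materials.
; After res-in it calls the parent and sets the BP_UseCCPM material flag.
; Each tick it writes ear-base socket locations into the BP_HandPosition parameters, repurposing those hand-named inputs, and ear-minus-head vectors into BP_OffsetVector.
; GameTimeFromBPForGlow only returns zero.
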(event EventOnPostResIn
  (|CallParentFunction)
  (Rendering|Material|SetScalarParameterValueonMaterials (CustomCharacter|GetSkeletalMeshForPartOnPlayerPawn) "BP_UseCCPM" 1.0))

(event EventTick
  (params (DeltaSeconds :type Float%20%28single-precision%29))
  (bind _returnvalue (CustomCharacter|GetSkeletalMeshForPartOnPlayerPawn))
  (bind _returnvalue_2 (Transformation|GetSocketLocation _returnvalue "dyn_ear_r_base"))
  (bind _returnvalue_3 (Transformation|GetSocketLocation _returnvalue "dyn_ear_l_base"))
  (bind _returnvalue_4 (Transformation|GetSocketLocation _returnvalue "head"))
  (Rendering|Material|SetVectorParameterValueonMaterials _returnvalue "BP_HandPosition_L" _returnvalue_3)
  (Rendering|Material|SetVectorParameterValueonMaterials _returnvalue "BP_HandPosition_R" _returnvalue_2)
  (Rendering|Material|SetVectorParameterValueonMaterials _returnvalue "BP_OffsetVector_L" (- _returnvalue_3 _returnvalue_4))
  (Rendering|Material|SetVectorParameterValueonMaterials _returnvalue "BP_OffsetVector_R" (- _returnvalue_2 _returnvalue_4)))

(fn GameTimeFromBPForGlow
  (params (A :type Float%20%28double-precision%29) (B :type Float%20%28double-precision%29))
  (return 0.0))

(fn ConstructionScript)
