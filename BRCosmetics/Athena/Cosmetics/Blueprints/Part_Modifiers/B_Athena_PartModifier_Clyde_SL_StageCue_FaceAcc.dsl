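; Cosmetic part modifier for the Clyde SL StageCue face accessory that drives material parameters on a player pawn's part mesh.
; After res-in it calls the parent and switches on BP_UseCCPM.
; Each tick it writes the head socket position and the neck_02-to-head offset into the materials, reusing right-hand parameter names (BP_HandPosition_R, BP_OffsetVector_R).
; GameTimeFromBPForGlow is a stub returning zero, and the construction script is empty.
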(event EventOnPostResIn
  (|CallParentFunction)
  (Rendering|Material|SetScalarParameterValueonMaterials (CustomCharacter|GetSkeletalMeshForPartOnPlayerPawn) "BP_UseCCPM" 1.0))

(event EventTick
  (params (DeltaSeconds :type Float%20%28single-precision%29))
  (bind _returnvalue (CustomCharacter|GetSkeletalMeshForPartOnPlayerPawn))
  (bind _returnvalue_2 (Transformation|GetSocketLocation _returnvalue "head"))
  (Rendering|Material|SetVectorParameterValueonMaterials _returnvalue "BP_HandPosition_R" _returnvalue_2)
  (Rendering|Material|SetVectorParameterValueonMaterials _returnvalue "BP_OffsetVector_R" (- _returnvalue_2 (Transformation|GetSocketLocation _returnvalue "neck_02"))))

(fn GameTimeFromBPForGlow
  (params (A :type Float%20%28double-precision%29) (B :type Float%20%28double-precision%29))
  (return 0.0))

(fn ConstructionScript)
