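; Cosmetic part modifier for the Conscience character part that plays a reactive material effect and sound on kills.
; It watches the kill stat, triggers an attached sound, and runs a timeline that drives three Reactive_Progress material scalars on the part mesh.
; A gate blocks retriggering until the timeline finishes.
; In the front end, the timeline starts without a kill and replays after a short delay as a preview.
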(event EventOnPostResIn
  (bind _returnvalue (CustomCharacter|GetSkeletalMeshForPartOnPlayerPawn))
  (|CallParentFunction)
  (Utilities|FlowControl|Sequence
    (:then_0
      (bind (_statname _statvalue) (Cosmetics|WatchCosmeticStat (CustomCharacter|GetAssociatedPlayerPawn) "(TagName=\"StatManager.AthenaKills\")")
        (:OnStatChanged
          (if (> _statvalue 0)
            (Utilities|FlowControl|Gate false :!label _label
              (:Exit
                (Utilities|FlowControl|Sequence
                  (:then_0
                    ;[0 0 0] Reactive "On Kill" Audio
                    (CallFunction|TriggerAudio)
                    (bind (_direction _step_a _step_b _step_c) (Timeline|Progress PlayFromStart)
                      (:Update
                        _)
                      (:Finished
                        (if (CustomCosmeticModifier|IsinFrontEnd)
                          (Utilities|FlowControl|Delay 1.5)
                          (Timeline|Progress PlayFromStart)
                          (else
                            (goto _label Open)))))
                    (Rendering|Material|SetScalarParameterValueonMaterials _returnvalue "Reactive_Progress_A" _step_a)
                    (Rendering|Material|SetScalarParameterValueonMaterials _returnvalue "Reactive_Progress_B" _step_b)
                    (Rendering|Material|SetScalarParameterValueonMaterials _returnvalue "Reactive_Progress_C" _step_c))
                  (:then_1
                    (goto _label Close)))
                (goto _label)))))))
    (:then_1
      (if (CustomCosmeticModifier|IsinFrontEnd)
        _)))
  (Timeline|Progress PlayFromStart))

(event Custom|Trigger%20Audio
  ;[0 0 0] Audio
  (Audio|SpawnSoundAttached (Variables|Default|GetSound) (CustomCharacter|GetSkeletalMeshForPartOnPlayerPawn)))

(event EventOnDayPhaseChanged
  (params (CurrentDayPhase :type EFortDayPhase%20Enum :default "Morning") (PreviousDayPhase :type EFortDayPhase%20Enum :default "Morning") (bAtCreation :type Boolean)))

(event EventOnNewPlayerWeaponType
  (params (NewType :type EFortItemType%20Enum :default "WorldItem") (OldType :type EFortItemType%20Enum :default "WorldItem")))

(event EventOnWeaponChanged
  (params (NewWeapon :type Fort%20Weapon%20Object%20Reference) (OldWeapon :type Fort%20Weapon%20Object%20Reference)))

(event EventOnWeaponFired)

(event EventOnNewWeapons
  (params (WeaponsAsItems :type Array%20of%20Fort%20Item%20Object%20References)))

(event EventOnCurieStateAttached
  (params (Tag :type Gameplay%20Tag%20Structure%20%28by%20ref%29)))

(event EventOnCurieStateDetached
  (params (Tag :type Gameplay%20Tag%20Structure%20%28by%20ref%29)))

(event EventOnBeginSkydiving)

(event EventOnEndSkydiving)

(event EventOnEnterVehicle
  (params (VehicleEntered :type Fort%20Vehicle%20Interface%20Interface%20%28by%20ref%29)))

(event EventOnExitVehicle)

(event EventOnTargetingOrFirstPersonCameraChanged)

(event EventOnHitPawn
  (params (SelfActor :type Actor%20Object%20Reference) (OtherActor :type Actor%20Object%20Reference) (NormalImpulse :type Vector :default "0, 0, 0") (Hit :type Hit%20Result%20Structure%20%28by%20ref%29)))

(fn ConstructionScript)
