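; Cosmetic character part modifier Blueprint for the female Constellation outfit.
; When the post-res-in event fires, it calls the parent, then attaches its left and right leg components to a character part skeletal mesh on the player pawn.
; They attach at thigh_l and thigh_r, keeping relative transforms and welding simulated bodies.
; The construction script only calls the parent.
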(event EventOnPostResIn
  (|CallParentFunction)
  (Transformation|AttachComponentToComponent (Variables|Default|GetLeg_LT) (CustomCharacter|GetSkeletalMeshForPartOnPlayerPawn) "thigh_l" "KeepRelative" "KeepRelative" "KeepRelative" true)
  (Transformation|AttachComponentToComponent (Variables|Default|GetLeg_RT) (CustomCharacter|GetSkeletalMeshForPartOnPlayerPawn) "thigh_r" "KeepRelative" "KeepRelative" "KeepRelative" true))

(fn ConstructionScript
  (|CallParentFunction))
