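; Cosmetic part modifier for the CopperMoggy character that enables render custom depth on its skeletal mesh for cape transparency.
; When customization finishes it turns custom depth on, sets a flag, and overrides the CenterOffset material parameter if displayed in the frontend.
; Tick re-enables custom depth if something overrides it, then disables itself once the actor is over one second old.
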(event EventOnCharacterCustomizationFinished
  (params (CharacterPartInfo :type Array%20of%20Applied%20Character%20Part%20Info%20Structures))
  (|CallParentFunction CharacterPartInfo)
  ; Turn on render custom depth for cape transparency
  (Rendering|SetRenderCustomDepth (CustomCharacter|GetSkeletalMeshForPartOnPlayerPawn) true)
  (Variables|Default|SetRender%20Custom%20Depth%20Pass%20On true)
  (if (Reactive|Frontend|IsCosmeticObjectDisplayedinFrontendContext 0)
    (Rendering|Material|SetColorParameterValueonMaterials (CustomCharacter|GetSkeletalMeshforPartonOwner) "CenterOffset" "(R=0.000000,G=0.000000,B=115.000000,A=1.000000)")))

(event EventTick
  (params (DeltaSeconds :type Float%20%28single-precision%29))
  ; Toggle custom depth on if it got overrided
  (|CallParentFunction DeltaSeconds)
  (if (> (Actor|GetGameTimeSinceCreation) 1.0)
    (Actor|Tick|SetActorTickEnabled false)
    (elif (and (== (Variables|/Script/Engine.PrimitiveComponent|GetbRenderCustomDepth (CustomCharacter|GetSkeletalMeshForPartOnPlayerPawn)) false) (Variables|Default|GetRender%20Custom%20Depth%20Pass%20On))
      (Rendering|SetRenderCustomDepth (CustomCharacter|GetSkeletalMeshForPartOnPlayerPawn) true))))

(fn ArrayIsValid
  (params (Name%20Array :type Array%20of%20Names))
  (return (> (Utilities|Array|Length Name%20Array) 0)))

(fn ConstructionScript
  (|CallParentFunction))
