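; Corny hair cosmetic part modifier that changes hair material and particle effects as a watched stat reaches new stages.
; Attaches idle VFX to the head, setting it by material variant in the frontend and hiding the stage elements otherwise.
; Stage activation reveals that stage's element, fires VFX and starts a six-second PopCorn timer.
; PopCorn replays VFX by stage; only stage 2 resets to zero and rebases thresholds on the stored stat.
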(event EventOnPostResIn
  (bind _returnvalue (CallFunction|GetIdleVFXComponent))
  (bind _returnvalue_2 (CallFunction|GetIdleVFXComponent))
  (bind _returnvalue_3 (CallFunction|GetIdleVFXComponent))
  (bind _returnvalue_4 (CustomCharacter|GetSkeletalMeshForPartOnPlayerPawn))
  (|CallParentFunction)
  (Transformation|AttachComponentToComponent (CallFunction|GetIdleVFXComponent) (Customization|GetSkeletalMeshforPartType (CustomCharacter|GetAssociatedPlayerPawn) "Head" :!class /Script/FortniteGame.FortPlayerPawn) "head" "SnapToTarget" "SnapToTarget" "KeepRelative" true)
  (if (CustomCosmeticModifier|IsinFrontEnd)
    (bind _index (CosmeticLoadout|GetVariantChannelActiveIndexforSlotActor "(TagName=\"LoadoutSlot.Character\")" (CustomCharacter|GetAssociatedPlayerPawn) "(TagName=\"Cosmetics.Variant.Channel.Material\")" (Utilities|Array|MakeArray "(TagName=\"Cosmetics.Variant.Property.Mat0\")" "(TagName=\"Cosmetics.Variant.Property.Mat1\")" "(TagName=\"Cosmetics.Variant.Property.Mat2\")" "(TagName=\"Cosmetics.Variant.Property.Mat3\")")))
    (switch int _index
      (:Default
        (Development|PrintString "Default"))
      (:0
        (Development|PrintString "Mat0"))
      (:1
        (Development|PrintString "Mat1")
        (Effects|Components|ParticleSystem|SetFloatParameter _returnvalue_3 "Switch" 0.0)
        (Components|Activation|Activate _returnvalue_3 true))
      (:2
        (Development|PrintString "Mat2")
        (Effects|Components|ParticleSystem|SetFloatParameter _returnvalue_2 "Switch" 1.0)
        (Components|Activation|Activate _returnvalue_2 true))
      (:3
        (Development|PrintString "Mat3")
        (Effects|Components|ParticleSystem|SetFloatParameter _returnvalue "Switch" 2.0)
        (Components|Activation|Activate _returnvalue true)))
    (else
      (Rendering|Material|SetScalarParameterValueonMaterials _returnvalue_4 "Hide Element 07" 1.0)
      (Rendering|Material|SetScalarParameterValueonMaterials _returnvalue_4 "Hide Element 08" 1.0)
      (Rendering|Material|SetScalarParameterValueonMaterials _returnvalue_4 "Hide Element 09" 1.0))))

(event EventOnCosmeticStateActivated
  (params (Stat%20Name :type Name :default "None") (Stat%20Value :type Integer) (STAGE%20Num%20Being%20Activated :type Integer))
  (bind _returnvalue (CustomCharacter|GetSkeletalMeshForPartOnPlayerPawn))
  (bind _returnvalue_2 (CustomCharacter|GetSkeletalMeshForPartOnPlayerPawn))
  (bind _returnvalue_3 (CallFunction|GetIdleVFXComponent))
  (bind _returnvalue_4 (CustomCharacter|GetSkeletalMeshForPartOnPlayerPawn))
  (|CallParentFunction Stat%20Name Stat%20Value STAGE%20Num%20Being%20Activated)
  (Development|PrintString "Corny is HIT!" true true "(R=1.000000,G=1.000000,B=1.000000,A=1.000000)")
  (Variables|Default|SetStage%20Value STAGE%20Num%20Being%20Activated)
  (Variables|Default|SetStat%20Value Stat%20Value)
  (Utilities|FlowControl|Sequence
    (:then_0
      (switch int STAGE%20Num%20Being%20Activated
        (:0
          (Development|PrintString "<<< Entering CORNY HAIR LEVEL 0  >>>>" true true "(R=1.000000,G=0.716152,B=0.000000,A=1.000000)"))
        (:1
          (Development|PrintString "<<< Entering CORNY HAIR LEVEL 1  >>>>" true true "(R=1.000000,G=0.487134,B=0.000000,A=1.000000)")
          (Rendering|Material|SetScalarParameterValueonMaterials _returnvalue "Hide Element 07" 0.0)
          (Rendering|Material|SetScalarParameterValueonMaterials _returnvalue "Hide Element 08" 1.0)
          (Rendering|Material|SetScalarParameterValueonMaterials _returnvalue "Hide Element 09" 1.0))
        (:2
          (Development|PrintString "<<< Entering CORNY HAIR LEVEL 2  >>>>" true true "(R=1.000000,G=0.247644,B=0.000000,A=1.000000)")
          (Rendering|Material|SetScalarParameterValueonMaterials _returnvalue_2 "Hide Element 07" 1.0)
          (Rendering|Material|SetScalarParameterValueonMaterials _returnvalue_2 "Hide Element 08" 0.0)
          (Rendering|Material|SetScalarParameterValueonMaterials _returnvalue_2 "Hide Element 09" 1.0))
        (:3
          (Development|PrintString "<<< Entering CORNY HAIR LEVEL 3  >>>>" true true "(R=1.000000,G=0.045875,B=0.000000,A=1.000000)")
          (Rendering|Material|SetScalarParameterValueonMaterials _returnvalue_4 "Hide Element 07" 1.0)
          (Rendering|Material|SetScalarParameterValueonMaterials _returnvalue_4 "Hide Element 08" 1.0)
          (Rendering|Material|SetScalarParameterValueonMaterials _returnvalue_4 "Hide Element 09" 0.0)))
      (Effects|Components|ParticleSystem|SetFloatParameter _returnvalue_3 "Switch" 0.0)
      (Components|Activation|Activate _returnvalue_3 true))
    (:then_1
      (Utilities|Time|SetTimerbyEvent (EventDispatchers|CreateEvent :!event Custom|PopCorn) 6.0))))

(event Custom|PopCorn
  (bind _returnvalue (CallFunction|GetIdleVFXComponent))
  (bind _returnvalue_2 (CallFunction|GetIdleVFXComponent))
  (bind _returnvalue_3 (CallFunction|GetIdleVFXComponent))
  (bind _returnvalue_4 (CustomCharacter|GetSkeletalMeshForPartOnPlayerPawn))
  (bind _stat_thresholds (Variables|PartModifierControls|WatchCosmeticStat|GetStat%20Thresholds))
  (bind _stat_value (Variables|Default|GetStat%20Value))
  (switch int (Variables|Default|GetStage%20Value)
    (:0
      _)
    (:1
      _)
    (:2
      (Effects|Components|ParticleSystem|SetFloatParameter _returnvalue_3 "Switch" 0.5)
      (Components|Activation|Activate _returnvalue_3 true)
      (Variables|PartModifierControls|WatchCosmeticStat|SetCurrent%20Active%20Stage%20Number 0)
      (Development|PrintString "<<< CORNY LEVEL RESET TO 0  >>>>" true true "(R=0.000000,G=1.000000,B=0.315753,A=1.000000)")
      (Rendering|Material|SetScalarParameterValueonMaterials _returnvalue_4 "Hide Element 07" 1.0)
      (Rendering|Material|SetScalarParameterValueonMaterials _returnvalue_4 "Hide Element 08" 1.0)
      (Rendering|Material|SetScalarParameterValueonMaterials _returnvalue_4 "Hide Element 09" 1.0)
      (Utilities|Array|SetArrayElem _stat_thresholds 0 _stat_value)
      (Utilities|Array|SetArrayElem _stat_thresholds 1 (+ _stat_value 20))
      (Utilities|Array|SetArrayElem _stat_thresholds 2 (+ _stat_value 40))
      (Utilities|Array|SetArrayElem _stat_thresholds 3 (+ _stat_value 60)))
    (:3
      (Effects|Components|ParticleSystem|SetFloatParameter _returnvalue "Switch" 1.0)
      (Components|Activation|Activate _returnvalue true)))
  (Effects|Components|ParticleSystem|SetFloatParameter _returnvalue_2 "Switch" 0.0)
  (Components|Activation|Activate _returnvalue_2 true))

(fn ConstructionScript
  (|CallParentFunction))
