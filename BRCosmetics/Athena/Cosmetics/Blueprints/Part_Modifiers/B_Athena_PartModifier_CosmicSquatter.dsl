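; Cosmetic part modifier for CosmicSquatter that feeds the player pawn's rotation to the materials on its skeletal mesh part.
; Outside the front end, a looping 60 Hz timer samples pawn rotation and adds yaw changes to a running total, compensating for wraparound jumps to keep it continuous.
; Only when yaw changes, it writes roll, pitch and accumulated yaw to the ActorRotationAmount vector parameter.
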(event EventOnPostResIn
  (|CallParentFunction)
  (if (not (CustomCosmeticModifier|IsinFrontEnd))
    (Utilities|Time|SetTimerbyEvent (delegate Custom|UpdateRotation) (/ 1.0 6e+01) true)))

(event Custom|UpdateRotation
  (bind (_returnvalue_roll _returnvalue_pitch) (Math|Rotator|BreakRotator (Transformation|GetActorRotation (CustomCharacter|GetAssociatedPlayerPawn))))
  (bind (_returnvalue_roll_2 _returnvalue_pitch_2 _returnvalue_yaw) (Math|Rotator|BreakRotator (Transformation|GetActorRotation (CustomCharacter|GetAssociatedPlayerPawn))))
  (bind _rotationbefore (Variables|Default|GetRotationBefore))
  (bind _rotationnow (Variables|Default|GetRotationNow))
  (bind _returnvalue (- _rotationnow _rotationbefore))
  (Variables|Default|SetRotationNow (+ _returnvalue_yaw 1.8e+02)) ; get rotation yaw in 0-360 range
  (if (!= (Variables|Default|GetRotationNow) (Variables|Default|GetRotationBefore))
    (if (>= (Math|Float|Absolute(Float) _returnvalue) 181.0)
      (Variables|Default|SetRotationAccumulated (+ (- (- (Variables|Default|GetRotationNow) _returnvalue) (Variables|Default|GetRotationBefore)) (Variables|Default|GetRotationAccumulated))) ; accumulate rotation yaw - rotation reset
      (else
        (Variables|Default|SetRotationAccumulated (+ (- (Variables|Default|GetRotationNow) (Variables|Default|GetRotationBefore)) (Variables|Default|GetRotationAccumulated))) ; accumulate rotation yaw
      ))
    (Variables|Default|SetRotationBefore (Variables|Default|GetRotationNow)) ; store current rotation value for next update
    (Rendering|Material|SetVectorParameterValueonMaterials (CustomCharacter|GetSkeletalMeshForPartOnPlayerPawn) "ActorRotationAmount" (Math|Vector|MakeVector _returnvalue_roll _returnvalue_pitch (Variables|Default|GetRotationAccumulated)))))

(fn ConstructionScript
  (|CallParentFunction))
