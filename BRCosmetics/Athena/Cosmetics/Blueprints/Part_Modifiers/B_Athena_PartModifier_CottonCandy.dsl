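; Cosmetic part modifier for the Cotton Candy back bling, driving its countdown, surprise animation and open/close effects.
; In the front end, a timeline writes a rounded countdown to a material scalar, then sets the AnimBP's surprise flag and, shortly after, its physics flag.
; In all contexts it binds AnimBP dispatchers to events that print debug text, attach and activate Niagara effects, and play sounds outside frontend display.
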(event EventOnPostResIn
  (|CallParentFunction)
  (Utilities|FlowControl|Sequence
    (:then_0
      (if (CustomCosmeticModifier|IsinFrontEnd)
        (bind (_direction _countdown) (Timeline|Timeline_0 PlayFromStart)
          (:Update
            (Rendering|Material|SetScalarParameterValueonMaterials (CustomCharacter|GetSkeletalMeshForPartOnPlayerPawn) "CC_Number_Offset_Seconds" (Math|Conversions|ToFloat(Integer) (Math|Float|Round _countdown))))
          (:Suprise%20Triger
            (bind _asf_med_cotton_candy_pack_anim_bp (Utilities|Casting|CastToF_MED_Cotton_Candy_Pack_AnimBP (Components|SkeletalMesh|GetAnimInstance (Customization|GetSkeletalMeshforPartType (CustomCharacter|GetAssociatedPlayerPawn) (CustomCharacter|GetCharacterPartType) :!class /Script/FortniteGame.FortPlayerPawn))))
            (Variables|/BRCosmetics/Accessories/FORT_Backpacks/Backpack_F_MED_Cotton_Candy/Meshes/F_MED_Cotton_Candy_Pack_AnimBP.F_MED_Cotton_Candy_Pack_AnimBP_C|SetSurprise_Out true _asf_med_cotton_candy_pack_anim_bp)
            (Utilities|FlowControl|Delay 0.2)
            (Variables|/BRCosmetics/Accessories/FORT_Backpacks/Backpack_F_MED_Cotton_Candy/Meshes/F_MED_Cotton_Candy_Pack_AnimBP.F_MED_Cotton_Candy_Pack_AnimBP_C|Setphysics_out true _asf_med_cotton_candy_pack_anim_bp)))))
    (:then_1
      (bind _asf_med_cotton_candy_pack_anim_bp_2 (Utilities|Casting|CastToF_MED_Cotton_Candy_Pack_AnimBP (Components|SkeletalMesh|GetAnimInstance (Customization|GetSkeletalMeshforPartType (CustomCharacter|GetAssociatedPlayerPawn) (CustomCharacter|GetCharacterPartType) :!class /Script/FortniteGame.FortPlayerPawn))))
      (Default|BindEventtoSurpriseFX _asf_med_cotton_candy_pack_anim_bp_2 (EventDispatchers|CreateEvent :!event Custom|Surprise%20FX))
      (Default|BindEventtoCloseFX _asf_med_cotton_candy_pack_anim_bp_2 (EventDispatchers|CreateEvent :!event Custom|Close%20FX)))))

(event Custom|Surprise%20FX
  (bind _returnvalue (CustomCharacter|GetAssociatedPlayerPawn))
  (bind _rootcomponent (Variables|/Script/Engine.Actor|GetRootComponent _returnvalue))
  (Development|PrintString "Surprise FX" true true "(R=1.000000,G=0.000000,B=0.000000,A=1.000000)")
  (Transformation|AttachComponentToComponent (Variables|Default|GetNS_Backpack_F_MED_Cotton_Candy_Surprise) (CustomCharacter|GetSkeletalMeshForPartOnPlayerPawn) "root" "KeepRelative" "KeepRelative" "KeepRelative" true)
  (Components|Activation|Activate (Variables|Default|GetNS_Backpack_F_MED_Cotton_Candy_Surprise) true)
  (if (not (Reactive|Frontend|IsCosmeticObjectDisplayedinFrontendContext _rootcomponent))
    (Audio|SpawnSoundAttached "/BRCosmetics/Sounds/Backblings/CottonCandy/BackBling_CottonCandy_Open_Cue.BackBling_CottonCandy_Open_Cue" _rootcomponent)))

(event Custom|Close%20FX
  (bind _returnvalue (CustomCharacter|GetAssociatedPlayerPawn))
  (bind _rootcomponent (Variables|/Script/Engine.Actor|GetRootComponent _returnvalue))
  (Development|PrintString "Close FX" true true "(R=1.000000,G=0.000000,B=0.000000,A=1.000000)")
  (Transformation|AttachComponentToComponent (Variables|Default|GetNS_Backpack_F_MED_Cotton_Candy_Close) (CustomCharacter|GetSkeletalMeshForPartOnPlayerPawn) "root" "KeepRelative" "KeepRelative" "KeepRelative" true)
  (Components|Activation|Activate (Variables|Default|GetNS_Backpack_F_MED_Cotton_Candy_Close) true)
  (if (not (Reactive|Frontend|IsCosmeticObjectDisplayedinFrontendContext _rootcomponent))
    (Audio|SpawnSoundAttached "/BRCosmetics/Sounds/Backblings/CottonCandy/BackBling_CottonCandy_Close_Cue.BackBling_CottonCandy_Close_Cue" _rootcomponent)))

(fn ConstructionScript)
