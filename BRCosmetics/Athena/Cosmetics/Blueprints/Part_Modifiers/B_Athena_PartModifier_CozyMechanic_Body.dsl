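; Body part modifier for the CozyMechanic cosmetic that adds reactive emissive color and idle VFX.
; In game it watches the player's kill stat; above zero, a timeline lerps mesh material emissive parameters toward reactivity colors, with separate helmet color, and activates the VFX once.
; When frontend reactivity applies, a delayed timeline previews the effect, replays after a delay, and toggles the VFX through event tracks.
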(event EventOnPostResIn
  (bind _returnvalue (CustomCharacter|GetSkeletalMeshForPartOnPlayerPawn))
  (bind _returnvalue_2 (CustomCharacter|GetSkeletalMeshForPartOnPlayerPawn))
  (bind _returnvalue_3 (CallFunction|GetIdleVFXComponent))
  (bind _reactivitycolor (Variables|Default|GetReactivityColor))
  (bind _reactivitycolor_2 (Variables|Default|GetReactivityColor))
  (|CallParentFunction)
  (if (CallFunction|ShouldDisplayFrontendReactivity)
    (Utilities|FlowControl|Delay 1.0)
    (bind (_direction _brightness _animationcurve) (Timeline|Timeline_1 PlayFromStart)
      (:Update
        (Rendering|Material|SetVectorParameterValueonMaterials _returnvalue_2 "EmissiveTint" (Math|Vector|Lerp(Vector) "1.000000,1.000000,1.000000" _reactivitycolor_2 _animationcurve))
        (Rendering|Material|SetVectorParameterValueonMaterials _returnvalue_2 "Emissive Color" (Math|Vector|Lerp(Vector) "1.000000,1.000000,1.000000" _reactivitycolor_2 _animationcurve))
        (Rendering|Material|SetVectorParameterValueonMaterials _returnvalue_2 "AdvancedEmissiveColor" (Math|Vector|Lerp(Vector) "0, 0, 0" (Variables|Default|GetReactivityColor_Helmet)))
        (Utilities|FlowControl|Delay 9.0)
        (Timeline|Timeline_1 PlayFromStart))
      (:FE_FX_appear
        (Components|Activation|Activate _returnvalue_3 true))
      (:FE_FX_disappear
        (Components|Activation|Deactivate _returnvalue_3)))
    (else
      (bind (_statname _statvalue) (Cosmetics|WatchCosmeticStat (CustomCharacter|GetAssociatedPlayerPawn) "(TagName=\"StatManager.AthenaKills\")" 0 false)
        (:OnStatChanged
          (if (> _statvalue 0)
            (Utilities|FlowControl|Sequence
              (:then_0
                (bind (_direction_2 _brightness_2 _animationcurve_2) (Timeline|Timeline PlayFromStart)
                  (:Update
                    (Rendering|Material|SetVectorParameterValueonMaterials _returnvalue "EmissiveTint" (Math|Vector|Lerp(Vector) "1.000000,1.000000,1.000000" _reactivitycolor _animationcurve_2))
                    (Rendering|Material|SetVectorParameterValueonMaterials _returnvalue "Emissive Color" (Math|Vector|Lerp(Vector) "1.000000,1.000000,1.000000" _reactivitycolor _animationcurve_2))
                    (Rendering|Material|SetVectorParameterValueonMaterials _returnvalue "AdvancedEmissiveColor" (Math|Vector|Lerp(Vector) "0, 0, 0" (Variables|Default|GetReactivityColor_Helmet) _animationcurve_2)))))
              (:then_1
                (Utilities|FlowControl|DoOnce false
                  (:Completed
                    (Components|Activation|Activate _returnvalue_3 true)
                    (Utilities|FlowControl|RetriggerableDelay 1.5)))))))))))

(event Custom|ReactivityFE
  (params (TriggerReactivity :type Boolean)))

(fn ConstructionScript
  (|CallParentFunction))
