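; Character part modifier for the CrawlyTech cosmetic that removes the fluffy overlay material from the head when fur is enabled.
; After customization finishes, it runs the parent logic, then clears the overlay material on the player pawn's Head skeletal mesh.
; The clear is gated on the return value of the r.ShellMesh.Supported CVar lookup, not on the numeric value it outputs.
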(event EventOnCharacterCustomizationFinished
  (params (CharacterPartInfo :type Array%20of%20Applied%20Character%20Part%20Info%20Structures))
  (bind (_outvalue _returnvalue) (Fortnite|GetNumericCVar "r.ShellMesh.Supported"))
  ; Remove fluffy overlay when fur is enabled
  (|CallParentFunction CharacterPartInfo)
  (if _returnvalue
    (Rendering|Material|SetOverlayMaterial (Customization|GetSkeletalMeshforPartType (CustomCharacter|GetAssociatedPlayerPawn) "Head" :!class /Script/FortniteGame.FortPlayerPawn) 0 false 0 :!class /Script/Engine.MeshComponent)))

(fn ConstructionScript
  (|CallParentFunction))
