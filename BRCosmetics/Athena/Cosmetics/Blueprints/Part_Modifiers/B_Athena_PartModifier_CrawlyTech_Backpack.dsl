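; Cosmetic part modifier for the CrawlyTech backpack that drives its material effects once the post-res-in event fires.
; In the front end, it sets emissive strength on one dynamic material and runs a delayed preview sequence.
; In game, it instead watches the pawn's kill stat for reactivity.
; It writes game time into the mesh materials' CT_GameTimeFromBP parameter, apparently as a timestamp for shader-driven animation.
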(event EventOnPostResIn
  (|CallParentFunction)
  (if (CustomCosmeticModifier|IsinFrontEnd)
    (Rendering|Material|SetScalarParameterValue (Class|BTechArtGlobalLibrary|GetDynamicMaterialInstance (CustomCharacter|GetSkeletalMeshForPartOnPlayerPawn) 2) "EmissiveStrength" 0.6 :!class /Script/Engine.MaterialInstanceDynamic)
    ; FE preview cycle
    (Utilities|FlowControl|Delay 0.5)
    (Utilities|FlowControl|Sequence
      (:then_0
        _)
      (:then_1
        (Utilities|FlowControl|Delay 7.5)))
    (else
      ; In game reactivity
      (Cosmetics|WatchCosmeticStat (CustomCharacter|GetAssociatedPlayerPawn) "(TagName=\"StatManager.AthenaKills\")" 0 false
        (:OnStatChanged
          _))))
  (Rendering|Material|SetScalarParameterValueonMaterials (CustomCharacter|GetSkeletalMeshForPartOnPlayerPawn) "CT_GameTimeFromBP" (Utilities|Time|GetGameTimeinSeconds)))

(fn ConstructionScript
  (|CallParentFunction))
