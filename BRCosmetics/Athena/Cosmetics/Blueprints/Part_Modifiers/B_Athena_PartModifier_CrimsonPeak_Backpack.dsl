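; Cosmetic part modifier for the CrimsonPeak backpack that makes its materials and idle VFX react to player damage.
; After initialization it resets the VFX Health parameter and checks the backpack's material variant, enabling the reaction only for the first.
; The front end plays a preview timeline once, while in game a DamageReceived stat watch replays another.
; Both timelines drive the IsDamagedFromBP material scalar and the Health particle parameter.
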(event EventOnPostResIn
  (bind _returnvalue (CustomCharacter|GetAssociatedPlayerPawn))
  (|CallParentFunction)
  (CallFunction|GetParts)
  (Effects|Components|ParticleSystem|SetFloatParameter (CallFunction|GetIdleVFXComponent) "Health" 1.0)
  (bind _asplayer_pawn_athena (Utilities|Casting|CastToPlayerPawn_Athena _returnvalue))
  (bind _index (CosmeticLoadout|GetVariantChannelActiveIndexforSlotActor "(TagName=\"LoadoutSlot.Backpack\")" _asplayer_pawn_athena "(TagName=\"Cosmetics.Variant.Channel.Material\")" (Utilities|Array|MakeArray "(TagName=\"Cosmetics.Variant.Property.Mat0\")" "(TagName=\"Cosmetics.Variant.Property.Mat1\")") false false))
  (switch int _index
    (:0
      (if (CustomCosmeticModifier|IsinFrontEnd)
        (bind (_direction _damage _healthfx) (Timeline|FE PlayFromStart)
          (:Update
            (Rendering|Material|SetScalarParameterValueonMaterials (CustomCharacter|GetSkeletalMeshForPartOnPlayerPawn) "IsDamagedFromBP" _damage)
            (Effects|Components|ParticleSystem|SetFloatParameter (CallFunction|GetIdleVFXComponent) "Health" _healthfx)))
        (else
          (Cosmetics|WatchCosmeticStat _returnvalue "(TagName=\"StatManager.DamageReceived\")" 0 false
            (:OnStatChanged
              (bind (_direction_2 _damage_2 _healthfx_2) (Timeline|Timeline PlayFromStart)
                (:Update
                  (Rendering|Material|SetScalarParameterValueonMaterials (CustomCharacter|GetSkeletalMeshForPartOnPlayerPawn) "IsDamagedFromBP" _damage_2)
                  (Effects|Components|ParticleSystem|SetFloatParameter (CallFunction|GetIdleVFXComponent) "Health" _healthfx_2))))))))))

(event EventOnStatChanged
  (params (State%20Name :type Name :default "None") (Stat%20Value :type Integer))
  (|CallParentFunction State%20Name Stat%20Value))

(fn GetParts
  (bind _returnvalue (CustomCharacter|GetAssociatedPlayerPawn))
  (Variables|Default|Sethead (Customization|GetSkeletalMeshforPartType _returnvalue "Head" :!class /Script/FortniteGame.FortPlayerPawn))
  (Variables|Default|Setbody (Customization|GetSkeletalMeshforPartType _returnvalue "Body" :!class /Script/FortniteGame.FortPlayerPawn))
  (Variables|Default|Setface (Customization|GetSkeletalMeshforPartType _returnvalue "Face" :!class /Script/FortniteGame.FortPlayerPawn)))

(fn ConstructionScript
  (|CallParentFunction))
