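; Cosmetic part modifier for the CrimsonPeak Ship that shows a damage reaction on the part's mesh and idle VFX.
; On post res-in, it watches the pawn's DamageReceived cosmetic stat, using short delays and a gate to control the effect.
; The effect sets damage parameters on materials and particles to full, then a timeline drives both after a brief hold.
; In the front end it also plays a FrontEndTriggers timeline, whose trigger output is empty.
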(event EventOnPostResIn
  (bind _returnvalue (CallFunction|GetIdleVFXComponent))
  (bind _returnvalue_2 (CustomCharacter|GetSkeletalMeshForPartOnPlayerPawn))
  (|CallParentFunction)
  (Utilities|FlowControl|Sequence
    (:then_0
      (Cosmetics|WatchCosmeticStat (CustomCharacter|GetAssociatedPlayerPawn) "(TagName=\"StatManager.DamageReceived\")"
        (:then
          (Utilities|FlowControl|Delay 0.2))
        (:OnStatChanged
          (Utilities|FlowControl|Delay 0.1)))
      (Utilities|FlowControl|Gate true
        (:Exit
          (Rendering|Material|SetScalarParameterValueonMaterials _returnvalue_2 "IsDamagedFromBP" 1.0)
          (Effects|Components|ParticleSystem|SetFloatParameter _returnvalue "Damaged" 1.0)
          (Utilities|FlowControl|Delay 0.2)
          (bind (_direction _newtrack_0) (Timeline|Timeline PlayFromStart)
            (:Update
              (Rendering|Material|SetScalarParameterValueonMaterials _returnvalue_2 "IsDamagedFromBP" _newtrack_0)
              (Effects|Components|ParticleSystem|SetFloatParameter _returnvalue "Damaged" _newtrack_0))))))
    (:then_1
      (if (CustomCosmeticModifier|IsinFrontEnd)
        (Timeline|FrontEndTriggers Play
          (:Trigger))))))

(fn ConstructionScript
  (|CallParentFunction))
