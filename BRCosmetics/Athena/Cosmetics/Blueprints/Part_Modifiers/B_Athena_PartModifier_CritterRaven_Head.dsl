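; Part modifier Blueprint for the CritterRaven head cosmetic that drives its idle particle effect when the pawn lands.
; On landing it fetches the idle VFX component, calls the parent handler, and sets the IsJumping bool parameter true.
; After a short delay it resets that parameter to false, producing a brief pulse.
; The construction script only calls the parent.
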(event EventOnPawnLanded
  (bind _returnvalue (CallFunction|GetIdleVFXComponent))
  (|CallParentFunction)
  (Effects|Components|ParticleSystem|SetBoolParameter _returnvalue "IsJumping" true)
  (Utilities|FlowControl|Delay 0.25)
  (Effects|Components|ParticleSystem|SetBoolParameter _returnvalue "IsJumping" false))

(fn ConstructionScript
  (|CallParentFunction))
