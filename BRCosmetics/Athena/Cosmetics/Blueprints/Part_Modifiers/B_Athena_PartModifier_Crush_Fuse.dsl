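; Cosmetic part modifier for the Crush Fuse character part that toggles its VFX around emotes.
; Starting an emote activates the idle VFX component, and ending one deactivates it.
; On emote end it also spawns a Niagara smoke trail attached to the root of the part's skeletal mesh.
; On successful spawn, it passes the hidden-hand and hidden-foot flags to that effect as bool parameters.
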(event EventOnBeginEmote
  (params (MontageItemDef :type Fort%20Item%20Definition%20Object%20Reference) (PawnEmoting :type Fort%20Pawn%20Object%20Reference))
  (Components|Activation|Activate (CallFunction|GetIdleVFXComponent) false))

(event EventOnEndEmote
  (params (MontageItemDef :type Fort%20Item%20Definition%20Object%20Reference) (PawnEmoting :type Fort%20Pawn%20Object%20Reference))
  (Components|Activation|Deactivate (CallFunction|GetIdleVFXComponent))
  (bind _returnvalue (PartModifier|SpawnSystem|SpawnSystemAttachedPartModifier "/BRCosmetics/Characters/Player/Female/Medium/Bodies/F_MED_Crush/FX/NS_Crush_Fuse_SmokeTrail.NS_Crush_Fuse_SmokeTrail" (CustomCharacter|GetSkeletalMeshForPartOnPlayerPawn) "root" "0, 0, 0" "0, 0, 0" "SnapToTargetIncludingScale")
    (:Success
      (Effects|Components|ParticleSystem|SetBoolParameter _returnvalue "bisHandHidden" (Variables|HiddenPartCheck|GetbIsHandHidden))
      (Effects|Components|ParticleSystem|SetBoolParameter _returnvalue "bisFootHidden" (Variables|HiddenPartCheck|GetbIsFootHidden)))))

(fn ConstructionScript)
