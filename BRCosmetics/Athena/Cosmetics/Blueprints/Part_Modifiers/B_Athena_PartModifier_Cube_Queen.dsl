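; Cosmetic part modifier for Cube Queen that drives a material effect on the character part from the player's targeting and movement state.
; On targeting or first-person camera changes, it sets ToggleFromBP to 1 when grounded and targeting, or 0 when not grounded.
; Each write also stamps GameTimeFromBP with game time on the part mesh's MID.
; Both branches sit behind separate DoOnce gates with no visible reset.
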(event EventOnTargetingOrFirstPersonCameraChanged
  (bind _returnvalue (CustomCharacter|GetSkeletalMeshForPartOnPlayerPawn))
  (bind _mid (Material|GetMID _returnvalue))
  (bind _returnvalue_2 (CustomCharacter|GetAssociatedPlayerPawn))
  (bind _returnvalue_3 (Utilities|Time|GetGameTimeinSeconds))
  (|CallParentFunction)
  (if (AI|Components|NavMovement|IsMovingonGround (Variables|/Script/Engine.Character|GetCharacterMovement _returnvalue_2))
    (if (Variables|/Script/FortniteGame.FortPlayerPawn|GetbIsTargeting _returnvalue_2)
      (Utilities|FlowControl|DoOnce false
        (:Completed
          (Rendering|Material|SetScalarParameterValue _mid "GameTimeFromBP" _returnvalue_3 :!class /Script/Engine.MaterialInstanceDynamic)
          (Rendering|Material|SetScalarParameterValue _mid "ToggleFromBP" 1.0 :!class /Script/Engine.MaterialInstanceDynamic))))
    (Utilities|FlowControl|DoOnce false
      (:Completed
        (Rendering|Material|SetScalarParameterValue _mid "GameTimeFromBP" _returnvalue_3 :!class /Script/Engine.MaterialInstanceDynamic)
        (Rendering|Material|SetScalarParameterValue _mid "ToggleFromBP" 0.0 :!class /Script/Engine.MaterialInstanceDynamic)))))

(fn ConstructionScript
  (|CallParentFunction))
