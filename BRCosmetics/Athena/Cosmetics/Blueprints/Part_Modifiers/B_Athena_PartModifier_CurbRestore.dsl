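; CurbRestore cosmetic part modifier that fades in a "MatchTime" effect on character materials and idle VFX as the match progresses.
; In-match, Tick drives the Face mesh and particles with match time over a fade-in duration, then sets the part material to 1 and stops ticking.
; In the front end, a preview timeline drives the parameter after PostResIn.
; Landing briefly pulses the VFX's "bPawnLanded" flag.
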(event EventOnPostResIn
  (|CallParentFunction)
  (if (CustomCosmeticModifier|IsinFrontEnd)
    (bind (_direction _snow) (Timeline|FE_Preview PlayFromStart)
      (:Update
        (Rendering|Material|SetScalarParameterValueonMaterials (CustomCharacter|GetSkeletalMeshForPartOnPlayerPawn) "MatchTime" _snow)
        (Effects|Components|ParticleSystem|SetFloatParameter (CallFunction|GetIdleVFXComponent) "MatchTime" _snow)))))

(event EventTick
  (params (DeltaSeconds :type Float%20%28single-precision%29))
  (bind _returnvalue (CallFunction|GetIdleVFXComponent))
  (bind _amount_of_time_to_complete_fade_in_secs (Variables|Default|GetAmount%20of%20Time%20to%20Complete%20Fade%20in%20Secs))
  (|CallParentFunction DeltaSeconds)
  (bind _asfort_game_state (Utilities|Casting|CastToFortGameState (Game|GetGameState)))
  (if (CustomCosmeticModifier|IsinFrontEnd)
    _
    (elif (> (/ (GameState|GetMatchTime _asfort_game_state) _amount_of_time_to_complete_fade_in_secs) 1.0)
      (Actor|Tick|SetActorTickEnabled false)
      (Rendering|Material|SetScalarParameterValueonMaterials (CustomCharacter|GetSkeletalMeshForPartOnPlayerPawn) "MatchTime" 1.0)
      (else
        (Rendering|Material|SetScalarParameterValueonMaterials (Customization|GetSkeletalMeshforPartType (CustomCharacter|GetAssociatedPlayerPawn) "Face" :!class /Script/FortniteGame.FortPlayerPawn) "MatchTime" (/ (GameState|GetMatchTime _asfort_game_state) _amount_of_time_to_complete_fade_in_secs))
        (Effects|Components|ParticleSystem|SetFloatParameter _returnvalue "MatchTime" (/ (GameState|GetMatchTime _asfort_game_state) _amount_of_time_to_complete_fade_in_secs))
        (Effects|Components|ParticleSystem|SetFloatParameter _returnvalue "MatchTime" (/ (GameState|GetMatchTime _asfort_game_state) _amount_of_time_to_complete_fade_in_secs))))))

(event EventOnPawnLanded
  (bind _returnvalue (CallFunction|GetIdleVFXComponent))
  ; ON PAWN LANDED
  (|CallParentFunction)
  (Effects|Components|ParticleSystem|SetBoolParameter _returnvalue "bPawnLanded" true)
  (Utilities|FlowControl|Delay 0.1)
  (Effects|Components|ParticleSystem|SetBoolParameter _returnvalue "bPawnLanded" false))

(fn ConstructionScript
  (|CallParentFunction))
