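; Character part modifier that keeps custom depth rendering enabled on a cosmetic part's skeletal mesh.
; When character customization finishes, it turns on custom depth for the mesh and sets a flag recording this.
; In a frontend context, Tick re-enables custom depth if something overrides it, then disables itself once the actor is over one second old.
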(event EventTick
  (params (DeltaSeconds :type Float%20%28single-precision%29))
  ; Toggle custom depth on if it got overrided
  (|CallParentFunction DeltaSeconds)
  (if (Reactive|Frontend|IsCosmeticObjectDisplayedinFrontendContext 0)
    (if (> (Actor|GetGameTimeSinceCreation) 1.0)
      (Actor|Tick|SetActorTickEnabled false)
      (elif (and (== (Variables|/Script/Engine.PrimitiveComponent|GetbRenderCustomDepth (CustomCharacter|GetSkeletalMeshForPartOnPlayerPawn)) false) (Variables|Default|GetRenderCustomDepthPassOn))
        (Rendering|SetRenderCustomDepth (CustomCharacter|GetSkeletalMeshForPartOnPlayerPawn) true)))))

(event EventOnCharacterCustomizationFinished
  (params (CharacterPartInfo :type Array%20of%20Applied%20Character%20Part%20Info%20Structures))
  ; Turn on render custom depth in FE
  (|CallParentFunction CharacterPartInfo)
  (Rendering|SetRenderCustomDepth (CustomCharacter|GetSkeletalMeshForPartOnPlayerPawn) true)
  (Variables|Default|SetRenderCustomDepthPassOn true))

(fn ConstructionScript
  (|CallParentFunction))
