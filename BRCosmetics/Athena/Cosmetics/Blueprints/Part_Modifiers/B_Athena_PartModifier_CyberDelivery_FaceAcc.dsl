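; Character part modifier for the CyberDelivery face accessory cosmetic that drives reactive expressions through material scalar parameters on the part's skeletal mesh.
; After res-in, it watches the kill and damage-received stats; an increase sets a matching expression index and stamps game time for the material.
; Accumulated velocity above a threshold turns on the running parameter, which each reaction suppresses for a few seconds through an override variable.
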(event EventOnPostResIn
  (bind _returnvalue (CustomCharacter|GetAssociatedPlayerPawn))
  (bind _returnvalue_2 (CustomCharacter|GetCharacterPartType))
  (bind _returnvalue_3 (Customization|GetSkeletalMeshforPartType _returnvalue _returnvalue_2 :!class /Script/FortniteGame.FortPlayerPawn))
  (bind _returnvalue_4 (CustomCharacter|GetAssociatedPlayerPawn))
  (bind _returnvalue_5 (CustomCharacter|GetCharacterPartType))
  (bind _returnvalue_6 (Customization|GetSkeletalMeshforPartType _returnvalue_4 _returnvalue_5 :!class /Script/FortniteGame.FortPlayerPawn))
  (|CallParentFunction)
  (Utilities|FlowControl|Sequence
    (:then_0
      ;[1 0.6239 0] Elimination
      (bind (_statname _statvalue) (Cosmetics|WatchCosmeticStat _returnvalue "(TagName=\"StatManager.AthenaKills\")")
        (:OnStatChanged
          (if (> _statvalue (Variables|Default|GetKillCount))
            (Rendering|Material|SetScalarParameterValueonMaterials _returnvalue_3 "BP_ExpressionIndex" 2.0)
            (Rendering|Material|SetScalarParameterValueonMaterials _returnvalue_3 "BP_GetTime" (Utilities|Time|GetGameTimeinSeconds))
            (Variables|Default|SetKillCount _statvalue)
            (Variables|Default|SetOverrideSprint 1.0)
            (Utilities|FlowControl|Delay 5.0)
            (Variables|Default|SetOverrideSprint 0.0)))))
    (:then_1
      ;[1 0.007073 0] Take Damage
      (bind (_statname_2 _statvalue_2) (Cosmetics|WatchCosmeticStat _returnvalue_4 "(TagName=\"StatManager.DamageReceived\")")
        (:OnStatChanged
          (if (> _statvalue_2 (Variables|Default|GetDamageCount))
            (Rendering|Material|SetScalarParameterValueonMaterials _returnvalue_6 "BP_ExpressionIndex" 3.0)
            (Rendering|Material|SetScalarParameterValueonMaterials _returnvalue_6 "BP_GetTime" (Utilities|Time|GetGameTimeinSeconds))
            (Variables|Default|SetDamageCount _statvalue_2)
            (Variables|Default|SetOverrideSprint 1.0)
            (Utilities|FlowControl|Delay 5.0)
            (Variables|Default|SetOverrideSprint 0.0)))) ; Damaged
    )))

(event EventOnLinearVelocity
  (params (NormalizedOverallVelocity :type Float%20%28double-precision%29) (NormalizedWalkVelocity :type Float%20%28double-precision%29) (VelocityLength :type Float%20%28double-precision%29)))

(event EventOnAccumulatedVelocity
  (params (NormalizedAccumulatedVelocity :type Float%20%28double-precision%29) (AccumulatedVelocity :type Float%20%28double-precision%29))
  (bind _returnvalue (CustomCharacter|GetAssociatedPlayerPawn))
  (bind _returnvalue_2 (CustomCharacter|GetCharacterPartType))
  (bind _returnvalue_3 (Customization|GetSkeletalMeshforPartType _returnvalue _returnvalue_2 :!class /Script/FortniteGame.FortPlayerPawn))
  (if (>= NormalizedAccumulatedVelocity 0.6)
    (Rendering|Material|SetScalarParameterValueonMaterials _returnvalue_3 "BP_IsRunning" (- 1.0 (Variables|Default|GetOverrideSprint)))
    (else
      (Rendering|Material|SetScalarParameterValueonMaterials _returnvalue_3 "BP_IsRunning" 0.0))))

(fn ConstructionScript)
