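; Cosmetic part modifier for Cyclone that attaches an air-glow mesh to the character part's "pack" socket.
; In the frontend, a repeating timer hides the glow and deactivates idle VFX while the pawn is a hologram, and is cleared on pre-destruction.
; Day-phase changes set Master_Alpha from per-phase settings through a single FInterpTo, with game time as delta rather than a per-frame fade, and print a debug string.
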(event EventOnPostResIn
  (|CallParentFunction)
  (Transformation|AttachComponentToComponent (Variables|Default|GetAirGlowMesh) (CustomCharacter|GetSkeletalMeshForPartOnPlayerPawn) "pack" "KeepRelative" "KeepRelative" "KeepRelative" true)
  (if (Reactive|Frontend|IsCosmeticObjectDisplayedinFrontendContext 0)
    (bind _returnvalue (Utilities|Time|SetTimerbyEvent (EventDispatchers|CreateEvent :!event Custom|HologramCheck) 0.1 true))
    (Variables|Default|SetHologramCheckTimer _returnvalue)))

(event EventPreDestruction
  (|CallParentFunction)
  (Utilities|Time|ClearandInvalidateTimerbyHandle (Variables|Default|GetHologramCheckTimer)))

(event Custom|HologramCheck
  (bind _returnvalue (CustomCharacter|GetAssociatedPlayerPawn))
  (bind _returnvalue_2 (CallFunction|GetIdleVFXComponent))
  (bind _bisinfrontendhologram (Variables|/Script/FortniteGame.FortPlayerPawn|GetbIsInFrontEndHologram _returnvalue))
  (Development|SetHiddeninGame (Variables|Default|GetAirGlowMesh) _bisinfrontendhologram false)
  (if _bisinfrontendhologram
    (Components|Activation|Deactivate _returnvalue_2)
    (else
      (Components|Activation|Activate _returnvalue_2 false))))

(event EventOnDayPhaseChanged
  (params (CurrentDayPhase :type EFortDayPhase%20Enum :default "Morning") (PreviousDayPhase :type EFortDayPhase%20Enum :default "Morning") (bAtCreation :type Boolean))
  (bind _returnvalue (Utilities|Time|GetGameTimeinSeconds))
  (bind _airglowalpha (Variables|Default|GetAirGlowAlpha))
  (bind _todalphasettings (Variables|Default|GetTODAlphaSettings))
  (bind _fade_speed (Variables|Default|GetFade%20Speed))
  (|CallParentFunction CurrentDayPhase PreviousDayPhase bAtCreation)
  (bind _result (ArrayMacros|SafeArrayGet _todalphasettings (Math|Conversions|ToInteger(Byte) CurrentDayPhase))
    (:%20%20
      (bind _result_2 (ArrayMacros|SafeArrayGet _todalphasettings (Math|Conversions|ToInteger(Byte) PreviousDayPhase))
        (:%20%20
          (Rendering|Material|SetScalarParameterValueonMaterials (Variables|Default|GetAirGlowMesh) "Master_Alpha" (Math|Interpolation|FInterpTo (* _result_2 _airglowalpha) (* _airglowalpha _result) _returnvalue _fade_speed))
          (Development|PrintString (Utilities|String|Append (Utilities|String|EnumtoString CurrentDayPhase) " - " (Utilities|String|ToString(Float) (Math|Interpolation|FInterpTo (* _result_2 _airglowalpha) (* _airglowalpha _result) _returnvalue _fade_speed))) true true "(R=1.000000,G=0.000000,B=0.414569,A=1.000000)" 1e+01))))))

(fn ConstructionScript)
