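; Cosmetic part modifier for the Cyclone Space character part that plays a hit reaction when its pawn-hit event fires.
; Unless the reaction timeline is already playing, it spawns a Niagara effect attached to the hit actor's root, oriented along the impact normal.
; It passes the hit location in mesh-local space to the part's materials as HitVector, then the timeline animates their Radius parameter, scaled by a Radius variable.
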(event EventOnHitPawn
  (params (SelfActor :type Actor%20Object%20Reference) (OtherActor :type Actor%20Object%20Reference) (NormalImpulse :type Vector :default "0, 0, 0") (Hit :type Hit%20Result%20Structure%20%28by%20ref%29))
  (bind _returnvalue (CustomCharacter|GetSkeletalMeshForPartOnPlayerPawn))
  (bind (_bblockinghit _binitialoverlap _time _distance _location _impactpoint _normal _impactnormal _physmat _hitactor) (Collision|BreakHitResult Hit))
  (if (not (Components|Timeline|IsPlaying (Variables|B_Athena_PartModifier_Cyclone_Space|GetHit_Reaction)))
    (Effects|SpawnFXSystemAttached "/BRCosmetics/Characters/Player/Male/Medium/Bodies/M_MED_Cyclone/FX/NS_Cyclone_SpaceHit.NS_Cyclone_SpaceHit" (Variables|/Script/Engine.Actor|GetRootComponent _hitactor) "root" _location (Math|Rotator|MakeRotfromX _impactnormal) "1.000000,1.000000,1.000000" "KeepWorldPosition" true "AutoRelease")
    (Rendering|Material|SetVectorParameterValueonMaterials _returnvalue "HitVector" (Math|Transform|InverseTransformLocation (Transformation|GetWorldTransform _returnvalue) _location))
    (bind (_direction _newtrack_0) (Timeline|Hit_Reaction PlayFromStart)
      (:Update
        (Rendering|Material|SetScalarParameterValueonMaterials _returnvalue "Radius" (* _newtrack_0 (Variables|Default|GetRadius)))))))

(fn ConstructionScript)
