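; Character part modifier for the Cyclone Universe cosmetic that adds a point light and a hit reaction effect to the player's body.
; After post-res-in, it destroys the light on mobile or low detail mode, otherwise attaches it to the pelvis.
; Day phase changes adjust light intensity, dimming at night.
; Pawn hits spawn a Niagara effect and send the local impact point and a timeline-driven radius to materials, ignoring hits while the timeline plays.
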(event EventOnPostResIn
  (|CallParentFunction)
  (bind _returnvalue (Default|CheckForLowMode))
  (if (not _returnvalue)
    (Transformation|AttachComponentToComponent (Variables|Default|GetPointLight) (CustomCharacter|GetSkeletalMeshForPartOnPlayerPawn) "pelvis" "KeepRelative" "KeepRelative" "KeepRelative" true)))

(event EventOnDayPhaseChanged
  (params (CurrentDayPhase :type EFortDayPhase%20Enum :default "Morning") (PreviousDayPhase :type EFortDayPhase%20Enum :default "Morning") (bAtCreation :type Boolean))
  (bind _pointlight (Variables|Default|GetPointLight))
  (Utilities|IsValid _pointlight
    (:Is%20Valid
      (Rendering|Components|Light|SetIntensity _pointlight (Utilities|Select :Index CurrentDayPhase :Morning 5.0 :Day 5.0 :Evening 5.0 :Night 3.5)))))

(event EventOnHitPawn
  (params (SelfActor :type Actor%20Object%20Reference) (OtherActor :type Actor%20Object%20Reference) (NormalImpulse :type Vector :default "0, 0, 0") (Hit_bBlockingHit :type Boolean) (Hit_bInitialOverlap :type Boolean) (Hit_Time :type Float%20%28single-precision%29) (Hit_Distance :type Float%20%28single-precision%29) (Hit_Location :type Vector :default "0, 0, 0") (Hit_ImpactPoint :type Vector :default "0, 0, 0") (Hit_Normal :type Vector :default "0, 0, 0") (Hit_ImpactNormal :type Vector :default "0, 0, 0") (Hit_PhysMat :type Physical%20Material%20Object%20Reference) (Hit_HitActor :type Actor%20Object%20Reference) (Hit_HitComponent :type Primitive%20Component%20Object%20Reference) (Hit_HitBoneName :type Name :default "None") (Hit_BoneName :type Name :default "None") (Hit_HitItem :type Integer) (Hit_ElementIndex :type Integer) (Hit_FaceIndex :type Integer) (Hit_TraceStart :type Vector :default "0, 0, 0") (Hit_TraceEnd :type Vector :default "0, 0, 0"))
  (bind _returnvalue (CustomCharacter|GetSkeletalMeshForPartOnPlayerPawn))
  (if (not (Components|Timeline|IsPlaying (Variables|B_Athena_PartModifier_Cyclone_Universe|GetHit_Reaction)))
    (Utilities|FlowControl|Sequence
      (:then_0
        (Effects|SpawnFXSystemAttached "/BRCosmetics/Characters/Player/Male/Medium/Bodies/M_MED_Cyclone/FX/NS_Cyclone_SpaceHit.NS_Cyclone_SpaceHit" Hit_HitComponent "root" Hit_ImpactPoint (Math|Rotator|MakeRotfromX Hit_ImpactNormal) "1.000000,1.000000,1.000000" "KeepWorldPosition" true "AutoRelease"))
      (:then_1
        (Rendering|Material|SetVectorParameterValueonMaterials _returnvalue "HitVector" (Math|Transform|InverseTransformLocation (Transformation|GetWorldTransform _returnvalue) Hit_ImpactPoint))
        (bind (_direction _newtrack_0) (Timeline|Hit_Reaction PlayFromStart)
          (:Update
            (Rendering|Material|SetScalarParameterValueonMaterials _returnvalue "Radius" (* _newtrack_0 (Variables|Default|GetRadius)))))))))

(fn CheckforLowMode
  (bind _returnvalue (TouchPlatform|IsMobileGame))
  (bind _returnvalue_2 (Rendering|GetRenderingDetailMode))
  (bind _returnvalue_3 (< _returnvalue_2 1))
  (bind _returnvalue_4 (or _returnvalue _returnvalue_3))
  (if _returnvalue_4
    (Components|DestroyComponent (Variables|Default|GetPointLight))
    (else
      _))
  (return _returnvalue_4))

(fn ConstructionScript)
