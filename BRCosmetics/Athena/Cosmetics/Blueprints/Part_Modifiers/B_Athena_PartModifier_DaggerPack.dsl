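; Character part modifier for the DaggerPack back bling that reacts to changes in its dagger count.
; On pawn/part change and post res-in, it reads the count from the loadout's active variant index and caches the first value silently.
; Later changes print a debug string, spawn attached Niagara FX at Skull_jnt and play a sound.
; ApplyMaterialParameters flags indices at or above the count via material scalars, but nothing here calls it.
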(event EventOnDayPhaseChanged
  (params (CurrentDayPhase :type EFortDayPhase%20Enum :default "Morning") (PreviousDayPhase :type EFortDayPhase%20Enum :default "Morning") (bAtCreation :type Boolean)))

(event EventOnNewPlayerWeaponType
  (params (NewType :type EFortItemType%20Enum :default "WorldItem") (OldType :type EFortItemType%20Enum :default "WorldItem"))
  (|CallParentFunction NewType OldType) ; This node is disabled and will not be called.\nDrag off pins to build functionality.
)

(event EventOnWeaponChanged
  (params (NewWeapon :type Fort%20Weapon%20Object%20Reference) (OldWeapon :type Fort%20Weapon%20Object%20Reference)))

(event EventOnWeaponFired)

(event EventOnNewWeapons
  (params (WeaponsAsItems :type Array%20of%20Fort%20Item%20Object%20References)))

(event EventOnCurieStateAttached
  (params (Tag :type Gameplay%20Tag%20Structure%20%28by%20ref%29)))

(event EventOnCurieStateDetached
  (params (Tag :type Gameplay%20Tag%20Structure%20%28by%20ref%29)))

(event EventOnBeginSkydiving
  (|CallParentFunction) ; This node is disabled and will not be called.\nDrag off pins to build functionality.
)

(event EventOnEndSkydiving
  (|CallParentFunction) ; This node is disabled and will not be called.\nDrag off pins to build functionality.
)

(event EventOnEnterVehicle
  (params (VehicleEntered :type Fort%20Vehicle%20Interface%20Interface%20%28by%20ref%29))
  (|CallParentFunction VehicleEntered) ; This node is disabled and will not be called.\nDrag off pins to build functionality.
)

(event EventOnExitVehicle
  (|CallParentFunction) ; This node is disabled and will not be called.\nDrag off pins to build functionality.
)

(event EventOnTargetingOrFirstPersonCameraChanged)

(event EventOnHitPawn
  (params (SelfActor :type Actor%20Object%20Reference) (OtherActor :type Actor%20Object%20Reference) (NormalImpulse :type Vector :default "0, 0, 0") (Hit :type Hit%20Result%20Structure%20%28by%20ref%29)))

(event EventOnPawnAndPartChanged
  (Utilities|IsValid (CustomCharacter|GetAssociatedPlayerPawn)
    (:Is%20Valid
      (Utilities|FlowControl|Sequence
        (:then_0
          (bind _loadout_dagger_count (CallFunction|GetLoadoutDaggerCount))
          (Variables|Default|SetNew%20Dagger%20Count _loadout_dagger_count)
          (if (Variables|Default|GetDagger%20Count%20Initialized)
            (if (!= (Variables|Default|GetNew%20Dagger%20Count) (Variables|Default|GetDagger%20Count))
              (Development|PrintString (Utilities|String|Append "Dagger Count went from " (Utilities|String|ToString(Integer) (Variables|Default|GetDagger%20Count)) " to " (Utilities|String|ToString(Integer) (Variables|Default|GetNew%20Dagger%20Count)) " - spawning FX!"))
              (Variables|Default|SetDagger%20Count (Variables|Default|GetNew%20Dagger%20Count))
              (CallFunction|SpawnFX)
              (CallFunction|TriggerAudio))
            (else
              (Variables|Default|SetDagger%20Count (Variables|Default|GetNew%20Dagger%20Count))
              (Variables|Default|SetDagger%20Count%20Initialized true))))))))

(event EventOnPostResIn
  (Utilities|IsValid (CustomCharacter|GetAssociatedPlayerPawn)
    (:Is%20Valid
      (Utilities|FlowControl|Sequence
        (:then_0
          (bind _loadout_dagger_count (CallFunction|GetLoadoutDaggerCount))
          (Variables|Default|SetNew%20Dagger%20Count _loadout_dagger_count)
          (if (Variables|Default|GetDagger%20Count%20Initialized)
            (if (!= (Variables|Default|GetNew%20Dagger%20Count) (Variables|Default|GetDagger%20Count))
              (Development|PrintString (Utilities|String|Append "Dagger Count went from " (Utilities|String|ToString(Integer) (Variables|Default|GetDagger%20Count)) " to " (Utilities|String|ToString(Integer) (Variables|Default|GetNew%20Dagger%20Count)) " - spawning FX!"))
              (Variables|Default|SetDagger%20Count (Variables|Default|GetNew%20Dagger%20Count))
              (CallFunction|SpawnFX)
              (CallFunction|TriggerAudio))
            (else
              (Variables|Default|SetDagger%20Count (Variables|Default|GetNew%20Dagger%20Count))
              (Variables|Default|SetDagger%20Count%20Initialized true))))))))

(event Custom|TriggerAudio
  ;[0 0 0] Reactive Audio Setup
  (Audio|SpawnSoundAttached (Variables|Default|GetSound) (CustomCharacter|GetSkeletalMeshForPartOnPlayerPawn)))

(fn ApplyMaterialParameters
  (bind _output_get (LocalVariables|SetSkeletal%20Mesh (CustomCharacter|GetSkeletalMeshForPartOnPlayerPawn)))
  (Utilities|IsValid _output_get
    (:Is%20Valid
      (for (_array_element _array_index) (Variables|Default|GetMaterial%20Parameter%20Names)
        (Rendering|Material|SetScalarParameterValueonMaterials (LocalVariables|GetSkeletal%20Mesh) _array_element (Math|Conversions|ToFloat(Boolean) (>= _array_index (Variables|Default|GetDagger%20Count))))))))

(fn GetLoadoutDaggerCount
  (bind _output_get (LocalVariables|SetPlayer%20Pawn (CustomCharacter|GetAssociatedPlayerPawn)))
  (Utilities|IsValid _output_get
    (:Is%20Valid
      (bind _index (CosmeticLoadout|GetVariantChannelActiveIndexforItemActor (Variables|Default|GetBackpack%20Item) (LocalVariables|GetPlayer%20Pawn) (Variables|Default|GetBackpack%20Channel) (Variables|Default|GetBackpack%20Variant%20Channels)))
      (return _index))
    (:Is%20Not%20Valid
      (return 0))))

(fn SpawnFX
  (Effects|SpawnFXSystemAttached "/BRCosmetics/Accessories/FORT_Backpacks/Backpack_M_MED_DaggerPack/FX/NS_Backpack_DaggerPack.NS_Backpack_DaggerPack" (CustomCharacter|GetSkeletalMeshForPartOnPlayerPawn) "Skull_jnt"))

(fn ConstructionScript
  (|CallParentFunction))
