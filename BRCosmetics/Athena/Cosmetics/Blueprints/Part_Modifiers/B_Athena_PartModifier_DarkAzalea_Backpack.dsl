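; Part modifier Blueprint for the DarkAzalea backpack cosmetic, driving its splash, bubble and liquid effects.
; When the pawn lands, it spawns an attached Niagara splash at the "pack" point on the part's skeletal mesh.
; After res-in, a short retriggerable delay precedes enabling looping bubbles on the idle VFX component and activating it.
; In the front end, that same step also swaps the mesh's liquid material for a frontend-specific instance.
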(event EventOnPawnLanded
  (|CallParentFunction)
  (Effects|SpawnFXSystemAttached "/BRCosmetics/Accessories/FORT_Backpacks/Backpack_F_MED_DarkAzalea/FX/NS_Backpack_DarkAzalea_Splash.NS_Backpack_DarkAzalea_Splash" (CustomCharacter|GetSkeletalMeshForPartOnPlayerPawn) "pack"))

(event EventOnPostResIn
  (bind _returnvalue (CallFunction|GetIdleVFXComponent))
  (|CallParentFunction)
  (Utilities|FlowControl|RetriggerableDelay 0.2)
  (bind _asniagara_particle_system_component (Utilities|Casting|CastToNiagaraComponent _returnvalue))
  (Niagara|SetNiagaraVariable(Bool) _asniagara_particle_system_component "Use Loop Bubble" true)
  (Components|Activation|Activate _returnvalue true)
  (if (CustomCosmeticModifier|IsinFrontEnd)
    (Rendering|Material|SetMaterial (CustomCharacter|GetSkeletalMeshForPartOnPlayerPawn) 2 "/BRCosmetics/Accessories/FORT_Backpacks/Backpack_F_MED_DarkAzalea/FX/MI_Frontend_Backpack_DarkAzalea_Liquid.MI_Frontend_Backpack_DarkAzalea_Liquid")))

(fn ConstructionScript
  (|CallParentFunction))
