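; Cosmetic character part modifier for the DarkEagle Fire skin that adds smolder ash and flame flick effects to the player.
; After res-in, it attaches the smolder ash components to the left and right clavicles and the flame flicks to both upper arms of the part's skeletal mesh.
; Skydiving start hides only the right smolder ash and skydiving end shows it again; the other effects are untouched.
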(event EventOnPostResIn
  (bind _returnvalue (CustomCharacter|GetSkeletalMeshForPartOnPlayerPawn))
  (|CallParentFunction)
  (Transformation|AttachComponentToComponent (Variables|Default|GetP_DarkEagleFire_SmolderAsh_R) _returnvalue "clavicle_r" "KeepRelative" "KeepRelative" "KeepRelative" true)
  (Transformation|AttachComponentToComponent (Variables|Default|GetP_DarkEagleFire_SmolderAsh_L) _returnvalue "clavicle_l" "KeepRelative" "KeepRelative" "KeepRelative" true)
  (Transformation|AttachComponentToComponent (Variables|Default|GetFlame%20Flicks_L) _returnvalue "upperarm_l" "KeepRelative" "KeepRelative" "KeepRelative" true)
  (Transformation|AttachComponentToComponent (Variables|Default|GetFlame%20Flicks_R) _returnvalue "upperarm_r" "KeepRelative" "KeepRelative" "KeepRelative" true))

(event EventOnBeginSkydiving
  (|CallParentFunction)
  (Development|SetHiddeninGame (Variables|Default|GetP_DarkEagleFire_SmolderAsh_R) true false))

(event EventOnEndSkydiving
  (|CallParentFunction)
  (Development|SetHiddeninGame (Variables|Default|GetP_DarkEagleFire_SmolderAsh_R) false false))

(fn ConstructionScript)
