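; Part modifier for the DarkStancePater body cosmetic that adjusts its idle VFX and materials when a hologram material override is applied to the player pawn.
; After post-res-in, it binds the pawn's override applied and cleared events, gated by a frontend display context check.
; Both handlers call SetHologramOverrides, which toggles idle VFX visibility and, for holograms, puts an invisible dynamic material on mesh slot 2.
; Clearing does not restore that slot here.
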(event Custom|OnMaterialOverrideApplied_Event
  (params (AffectedPawn :type Fort%20Player%20Pawn%20Object%20Reference) (AppliedOverride_OverrideId :type Guid%20Structure) (AppliedOverride_Material :type Material%20Interface%20Soft%20Object%20Reference))
  ;[0 0 0] hologram
  (Development|PrintString "HologramMaterialApplied")
  (CallFunction|SetHologramOverrides true))

(event EventOnPostResIn
  (bind _returnvalue (CustomCharacter|GetAssociatedPlayerPawn))
  (|CallParentFunction)
  (if (Reactive|Frontend|IsCosmeticObjectDisplayedinFrontendContext 0)
    (EventDispatchers|AssignOnMaterialOverrideApplied _returnvalue (delegate Custom|OnMaterialOverrideApplied_Event))
    (EventDispatchers|AssignOnMaterialOverrideCleared _returnvalue (delegate Custom|OnMaterialOverrideCleared_Event))))

(event Custom|OnMaterialOverrideCleared_Event
  (params (AffectedPawn :type Fort%20Player%20Pawn%20Object%20Reference))
  ;[0 0 0] hologramClear
  (Development|PrintString "HologramMaterialRemoved")
  (CallFunction|SetHologramOverrides false))

(fn SetHologramOverrides
  (params (isHologram :type Boolean))
  (LocalVariables|SetisHolo isHologram)
  (Development|SetHiddeninGame (CallFunction|GetIdleVFXComponent) (LocalVariables|GetisHolo) false)
  (if (LocalVariables|GetisHolo)
    (bind _returnvalue (Rendering|Material|CreateDynamicMaterialInstance "/Game/Missions/Secondary/BluGlo/M_Invisible.M_Invisible" "None" "None" :!class /Script/Engine.KismetMaterialLibrary))
    (Rendering|Material|SetMaterial (CustomCharacter|GetSkeletalMeshForPartOnPlayerPawn) 2 _returnvalue)))

(fn ConstructionScript
  (|CallParentFunction))
