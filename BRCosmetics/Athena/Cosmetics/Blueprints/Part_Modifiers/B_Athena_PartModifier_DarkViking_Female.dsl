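; Cosmetic part modifier blueprint for the female Dark Viking character that manages a left-eye particle trail.
; On post-resurrection-in it calls the parent, then snaps the eye trail component to the FX_EyeTrail_L socket on the pawn's character part skeletal mesh.
; The trail is hidden when skydiving begins and shown again when it ends.
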(event EventOnPostResIn
  (|CallParentFunction)
  (Transformation|AttachComponentToComponent (Variables|Default|GetP_DarkViking_EyeTrail_L) (CustomCharacter|GetSkeletalMeshForPartOnPlayerPawn) "FX_EyeTrail_L" "SnapToTarget" "SnapToTarget" "SnapToTarget" true))

(event EventOnBeginSkydiving
  (Development|SetHiddeninGame (Variables|Default|GetP_DarkViking_EyeTrail_L) true false))

(event EventOnEndSkydiving
  (Development|SetHiddeninGame (Variables|Default|GetP_DarkViking_EyeTrail_L) false false))

(fn ConstructionScript)
