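; Cosmetic part modifier for the female Dark Viking Fire character that places glow and gem effects on the player's skeletal mesh.
; On post res-in it calls the parent, then attaches both eye glows to the head and the rear gem to the FX_RearGem socket.
; On targeting or first-person camera changes, it hides only the left eye glow for the local player while hip-fire first person is active.
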(event EventOnPostResIn
  (bind _returnvalue (CustomCharacter|GetSkeletalMeshForPartOnPlayerPawn))
  (|CallParentFunction)
  (Transformation|AttachComponentToComponent (Variables|Default|GetSM_Eye_Glow_L) _returnvalue "head" "KeepRelative" "KeepRelative" "KeepRelative" true)
  (Transformation|AttachComponentToComponent (Variables|Default|GetSM_Eye_Glow_R) _returnvalue "head" "KeepRelative" "KeepRelative" "KeepRelative" true)
  (Transformation|AttachComponentToComponent (Variables|Default|GetRearGem) _returnvalue "FX_RearGem" "KeepRelative" "KeepRelative" "KeepRelative" true))

(event EventOnWeaponTargetingChanged
  (params (bIsTargeting :type Boolean))
  (|CallParentFunction bIsTargeting))

(event EventOnTargetingOrFirstPersonCameraChanged
  (bind _returnvalue (CustomCharacter|GetAssociatedPlayerPawn))
  (|CallParentFunction)
  (if (Pawn|IsLocallyControlled _returnvalue)
    (Development|SetHiddeninGame (Variables|Default|GetSM_Eye_Glow_L) (Class|FortPlayerPawn|IsHipFireFirstPersonActive _returnvalue) false)))

(fn ConstructionScript)
