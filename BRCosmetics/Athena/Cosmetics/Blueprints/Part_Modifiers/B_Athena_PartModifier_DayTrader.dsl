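; Part modifier Blueprint for the DayTrader back bling cosmetic that plays a reactive sound and VFX response when a tracked stat, apparently kills, changes.
; For values above zero, it waits briefly, then spawns an attached sound cue on the idle VFX component.
; The stat modulo 15 feeds the cue's KillCount parameter, so the audio input cycles.
; A restarted timeline then drives the idle particle system's OnKill pulse, swipe and bounce parameters.
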(event EventOnStatChanged
  (params (State%20Name :type Name :default "None") (Stat%20Value :type Integer))
  (bind _returnvalue (CallFunction|GetIdleVFXComponent))
  (|CallParentFunction State%20Name Stat%20Value)
  (if (> Stat%20Value 0)
    (Utilities|FlowControl|Delay 0.5)
    (bind _returnvalue_2 (Audio|SpawnSoundAttached "/BRCosmetics/Sounds/Backblings/DayTrader/BackBling_DayTrader_Cue.BackBling_DayTrader_Cue" (CallFunction|GetIdleVFXComponent)))
    (Audio|Parameter|SetFloatParameter _returnvalue_2 "KillCount" (Math|Conversions|ToFloat(Integer) (Math|Integer|%25%28Integer%29 Stat%20Value 15)) :!class /Script/AudioExtensions.AudioParameterControllerInterface)
    (bind (_direction _pulse _swipe _bounce _bounce_delay) (Timeline|Timeline_0 PlayFromStart)
      (:Update
        (Effects|Components|ParticleSystem|SetFloatParameter _returnvalue "OnKill_Pulse" _pulse)
        (Effects|Components|ParticleSystem|SetFloatParameter _returnvalue "OnKill_Swipe" _swipe)
        (Effects|Components|ParticleSystem|SetFloatParameter _returnvalue "OnKill_Bounce" _bounce)
        (Effects|Components|ParticleSystem|SetFloatParameter _returnvalue "OnKill_Bounce_Delay" _bounce_delay)))))

(fn ConstructionScript
  (|CallParentFunction))
