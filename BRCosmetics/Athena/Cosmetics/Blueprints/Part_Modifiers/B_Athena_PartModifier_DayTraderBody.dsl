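; Cosmetic part modifier for the DayTrader body that adjusts the character part's material look in the front end.
; After the parent's post-res-in logic runs, it checks whether the pawn is in the front end.
; If so, it overrides scalar parameters on the part's skeletal mesh materials for highlight, fresnel, "Lasor" strength and color range, and surface opacity.
; Outside the front end it adds no changes.
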(event EventOnPostResIn
  (bind _returnvalue (CustomCharacter|GetSkeletalMeshForPartOnPlayerPawn))
  (|CallParentFunction)
  (if (CustomCosmeticModifier|IsinFrontEnd)
    (Rendering|Material|SetScalarParameterValueonMaterials _returnvalue "DT_HighlightMult" 3.0)
    (Rendering|Material|SetScalarParameterValueonMaterials _returnvalue "DT_FresnelExp" 0.1)
    (Rendering|Material|SetScalarParameterValueonMaterials _returnvalue "DT_LasorStrength" 1.5)
    (Rendering|Material|SetScalarParameterValueonMaterials _returnvalue "DT_LasorColorRange" 0.2)
    (Rendering|Material|SetScalarParameterValueonMaterials _returnvalue "DT_SurfaceOpacity" 0.5)))

(fn ConstructionScript
  (|CallParentFunction))
