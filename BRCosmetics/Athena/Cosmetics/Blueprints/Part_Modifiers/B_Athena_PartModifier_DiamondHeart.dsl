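; Part modifier Blueprint for the DiamondHeart cosmetic that adjusts its materials when the pawn or part changes.
; If the cosmetic is displayed in a frontend context, it loops over every material on the part's skeletal mesh.
; It gets a dynamic instance for each material slot and overrides the GS_BaseIntensity scalar parameter with a fixed value.
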(event EventOnPawnAndPartChanged
  (bind _returnvalue (CustomCharacter|GetSkeletalMeshForPartOnPlayerPawn))
  (|CallParentFunction)
  (if (Reactive|Frontend|IsCosmeticObjectDisplayedinFrontendContext 0)
    (for (_ _array_index) (Rendering|Material|GetMaterials _returnvalue)
      (Rendering|Material|SetScalarParameterValue (Class|BTechArtGlobalLibrary|GetDynamicMaterialInstance _returnvalue _array_index) "GS_BaseIntensity" 3.6 :!class /Script/Engine.MaterialInstanceDynamic))))

(fn ConstructionScript
  (|CallParentFunction))
