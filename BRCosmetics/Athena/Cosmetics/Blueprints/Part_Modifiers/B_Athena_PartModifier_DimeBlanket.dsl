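; Cosmetic part modifier Blueprint for the DimeBlanket Athena character part.
; After the post res-in event runs its parent logic, it checks whether the character is in the front end.
; If so, it sets the "FEAudio" scalar parameter to 1 on the materials of the part's skeletal mesh on the player pawn.
; The construction script only calls its parent.
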(event EventOnPostResIn
  (|CallParentFunction)
  (if (CustomCosmeticModifier|IsinFrontEnd)
    (Rendering|Material|SetScalarParameterValueonMaterials (CustomCharacter|GetSkeletalMeshForPartOnPlayerPawn) "FEAudio" 1.0)))

(fn ConstructionScript
  (|CallParentFunction))
