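; Cosmetic character part modifier for the DragonNinja Lightning effect, keeping a randomized flicker on the player's part mesh and body particles.
; BeginPlay rolls a small random time offset, which OnPostResIn writes to the part mesh's materials as RandomOffsetPerClient_Mesh.
; On each Parent_Spawn event from P_DragonNinja_OnBody, three offset, game-time sine waves are blended into the "LightColor" float parameter, intended to mirror the material's pulse.
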(event EventOnPostResIn
  (|CallParentFunction)
  (Rendering|Material|SetScalarParameterValueonMaterials (Customization|GetSkeletalMeshforPartType (CustomCharacter|GetAssociatedPlayerPawn) (CustomCharacter|GetCharacterPartType) :!class /Script/FortniteGame.FortPlayerPawn) "RandomOffsetPerClient_Mesh" (Variables|Default|GetRandomOffsetPerClient)))

(event EventBeginPlay
  ;[1 1 1] Set random time offset of the fx
  (|CallParentFunction)
  (Variables|Default|SetRandomOffsetPerClient (Math|Random|RandomFloatinRange 0.1 0.3)))

(event OnParticleSpawn(P_DragonNinja_OnBody)
  (params (EventName :type Name :default "None") (EmitterTime :type Float%20%28single-precision%29) (Location :type Vector :default "0, 0, 0") (Velocity :type Vector :default "0, 0, 0"))
  ;[1 1 1] Checks for event generator in Cascade and sets the float parameter
  (switch name EventName
    (:Parent_Spawn
      ;[1 1 1] Light intensity pulse same as the material, temp removed
      (Effects|Components|ParticleSystem|SetFloatParameter (Variables|Default|GetP_DragonNinja_OnBody) "LightColor" (Math|Float|Clamp(Float) (Math|Float|Lerp (Math|Trig|Sin(Radians) (* (+ (Math|Float|Fraction (/ (Utilities|Time|GetGameTimeinSeconds) 2.6)) (Math|Float|Lerp 0.0 1.0 (Variables|Default|GetRandomOffsetPerClient))) (* (Math|Trig|GetPI) 5.2))) (Math|Trig|Sin(Radians) (* (* (Math|Trig|GetPI) 3.14) (+ (Math|Float|Fraction (/ (Utilities|Time|GetGameTimeinSeconds) 1.57)) (Math|Float|Lerp 0.0 1.0 (Variables|Default|GetRandomOffsetPerClient))))) (Math|Float|Lerp 0.15 0.85 (Math|Trig|Sin(Radians) (* (* (Math|Trig|GetPI) 2.48) (+ (Math|Float|Fraction (/ (Utilities|Time|GetGameTimeinSeconds) 1.24)) (Math|Float|Lerp 0.0 1.0 (Variables|Default|GetRandomOffsetPerClient))))))))))))

(fn ConstructionScript)
