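; Part modifier Blueprint for the Dragon Racer back bling cosmetic that adds particle effects to the player's jumps and landings.
; On pawn jumped, it calls the parent handler, then spawns the jets emitter attached to the FX_Base socket on the part's skeletal mesh.
; On pawn landed, it does the same with a separate landing emitter, using the legacy part-modifier spawn call in both cases.
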(event EventOnPawnJumped
  (|CallParentFunction)
  (PartModifier|SpawnSystem|SpawnEmitterAttachedPartModifier(Legacy) "/BRCosmetics/Accessories/FORT_Backpacks/Backpack_DragonRacer/FX/P_Backpack_DragonRacer_Jets.P_Backpack_DragonRacer_Jets" (CustomCharacter|GetSkeletalMeshForPartOnPlayerPawn) "FX_Base"))

(event EventOnPawnLanded
  (|CallParentFunction)
  (PartModifier|SpawnSystem|SpawnEmitterAttachedPartModifier(Legacy) "/BRCosmetics/Accessories/FORT_Backpacks/Backpack_DragonRacer/FX/P_Backpack_DragonRacer_Landing.P_Backpack_DragonRacer_Landing" (CustomCharacter|GetSkeletalMeshForPartOnPlayerPawn) "FX_Base"))

(fn ConstructionScript)
