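; Cosmetic part modifier for the DriftWinter character body that plays an ice Niagara landing effect on the player.
; After res-in, it attaches the effect to the part's skeletal mesh root, then reactivates it whenever the pawn lands.
; Skydiving start and end run effects-quality checks and start or clear a looping timer.
; Each timer tick reactivates the same effect if the pawn's jump is providing force.
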(event EventOnPostResIn
  (|CallParentFunction)
  (Transformation|AttachComponentToComponent (Variables|Default|GetNS_DriftWinter_Landing) (CustomCharacter|GetSkeletalMeshForPartOnPlayerPawn) "root" "SnapToTarget" "SnapToTarget" "KeepRelative" true))

(event EventOnBeginSkydiving
  (|CallParentFunction)
  (Utilities|EffectsQualityCheck 0
    (:Cinematic
      _)
    (:Epic
      _)
    (:High
      _))
  (CallFunction|DropIceFX))

(event EventOnEndSkydiving
  (|CallParentFunction)
  (Utilities|EffectsQualityCheck 0
    (:Cinematic
      _)
    (:Epic
      _)
    (:High
      _))
  (Utilities|Time|ClearandInvalidateTimerbyHandle (Variables|Default|GetSkydiving%20Timer)))

(event Custom|Drop%20Ice%20FX
  (bind _returnvalue (Utilities|Time|SetTimerbyEvent (delegate Custom|Skydiving%20Activation) 0.2 true))
  (Variables|Default|SetSkydiving%20Timer _returnvalue))

(event Custom|Skydiving%20Activation
  (if (Character|IsJumpProvidingForce (CustomCharacter|GetAssociatedPlayerPawn))
    (Components|Activation|Activate (Variables|Default|GetNS_DriftWinter_Landing) true)))

(event EventOnPawnLanded
  (|CallParentFunction)
  (Components|Activation|Activate (Variables|Default|GetNS_DriftWinter_Landing) true))

(fn ConstructionScript
  (|CallParentFunction))
