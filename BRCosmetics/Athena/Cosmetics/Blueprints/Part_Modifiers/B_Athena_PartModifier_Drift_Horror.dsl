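; Cosmetic part modifier for the Drift Horror character that hides its idle VFX while parachuting with a specific glider.
; After res-in, it compares the pawn loadout's glider to a stored Glider ID and caches the match as a transforming-glider flag.
; It also binds the pawn's parachute begin and end events, which deactivate and reactivate the idle VFX component when that flag is set.
; All other overridden events are empty or disabled stubs.
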(event EventOnDayPhaseChanged
  (params (CurrentDayPhase :type EFortDayPhase%20Enum :default "Morning") (PreviousDayPhase :type EFortDayPhase%20Enum :default "Morning") (bAtCreation :type Boolean)))

(event EventOnNewPlayerWeaponType
  (params (NewType :type EFortItemType%20Enum :default "WorldItem") (OldType :type EFortItemType%20Enum :default "WorldItem"))
  (|CallParentFunction NewType OldType) ; This node is disabled and will not be called.\nDrag off pins to build functionality.
)

(event EventOnWeaponChanged
  (params (NewWeapon :type Fort%20Weapon%20Object%20Reference) (OldWeapon :type Fort%20Weapon%20Object%20Reference)))

(event EventOnWeaponFired)

(event EventOnPostResIn
  (bind _returnvalue (CustomCharacter|GetAssociatedPlayerPawn))
  (bind (_returnvalue_bannericonid _returnvalue_bannercolorid _returnvalue_skydivecontrail _returnvalue_glider) (Utilities|Struct|BreakFortAthenaLoadout (Cosmetics|GetPawnCosmeticLoadout _returnvalue false true)))
  (|CallParentFunction)
  (EventDispatchers|BindEventtoOnBeginParachuteMovementEvent _returnvalue (delegate Custom|Disable%20FX%20On%20Begin%20Parachute%20Movement))
  (EventDispatchers|BindEventtoOnEndParachuteMovementEvent _returnvalue (delegate Custom|Enable%20FX%20On%20Begin%20Parachute%20Movement))
  (Variables|Default|SetHasTransforming%20Glider (Utilities|Equal(Object) _returnvalue_glider (Variables|Default|GetGlider%20ID))))

(event EventOnNewWeapons
  (params (WeaponsAsItems :type Array%20of%20Fort%20Item%20Object%20References)))

(event EventOnCurieStateAttached
  (params (Tag :type Gameplay%20Tag%20Structure%20%28by%20ref%29)))

(event EventOnCurieStateDetached
  (params (Tag :type Gameplay%20Tag%20Structure%20%28by%20ref%29)))

(event Custom|Disable%20FX%20On%20Begin%20Parachute%20Movement
  (if (Variables|Default|GetHasTransforming%20Glider)
    (Components|Activation|Deactivate (CallFunction|GetIdleVFXComponent))))

(event EventOnBeginSkydiving
  (|CallParentFunction) ; This node is disabled and will not be called.\nDrag off pins to build functionality.
)

(event Custom|Enable%20FX%20On%20Begin%20Parachute%20Movement
  (if (Variables|Default|GetHasTransforming%20Glider)
    (Components|Activation|Activate (CallFunction|GetIdleVFXComponent) true)))

(event EventOnEndSkydiving
  (|CallParentFunction) ; This node is disabled and will not be called.\nDrag off pins to build functionality.
)

(event EventOnEnterVehicle
  (params (VehicleEntered :type Fort%20Vehicle%20Interface%20Interface%20%28by%20ref%29))
  (|CallParentFunction VehicleEntered) ; This node is disabled and will not be called.\nDrag off pins to build functionality.
)

(event EventOnExitVehicle
  (|CallParentFunction) ; This node is disabled and will not be called.\nDrag off pins to build functionality.
)

(event EventOnTargetingOrFirstPersonCameraChanged)

(event EventOnHitPawn
  (params (SelfActor :type Actor%20Object%20Reference) (OtherActor :type Actor%20Object%20Reference) (NormalImpulse :type Vector :default "0, 0, 0") (Hit :type Hit%20Result%20Structure%20%28by%20ref%29)))

(fn ConstructionScript
  (|CallParentFunction))
