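; DyeBlast cosmetic part modifier that plays a reactive glitch effect when a tracked stat changes, noted as eliminations.
; On res-in it creates and stores a dynamic holo material on slot 1 of the part's skeletal mesh.
; Stat changes restart the Play Glitch FX timeline, with play rate set to the inverse of Glitch Speed.
; When backpack frontend reactivity is shown, a delayed looping timer replays the same glitch as a preview.
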(event EventOnPostResIn
  (bind _returnvalue (CustomCharacter|GetSkeletalMeshForPartOnPlayerPawn))
  (bind _returnvalue_2 1)
  (|CallParentFunction)
  (bind _returnvalue_3 (Rendering|Material|CreateDynamicMaterialInstance _returnvalue _returnvalue_2 (Rendering|Material|GetMaterial _returnvalue _returnvalue_2) "None" :!class /Script/Engine.PrimitiveComponent))
  (bind _output_get (Variables|Default|SetHoloMat _returnvalue_3))
  (Rendering|Material|SetMaterial _returnvalue _returnvalue_2 _output_get)
  (if (Reactive|Frontend|ShouldDisplayFrontendReactivity 0 "Backpack")
    (Utilities|FlowControl|Delay 3.0)
    (CallFunction|FrontEndLoop)
    (Utilities|Time|SetTimerbyEvent (delegate Custom|FrontEnd%20Loop) 4.0 true)))

(event EventOnStatChanged
  (params (State%20Name :type Name :default "None") (Stat%20Value :type Integer))
  (bind _play_glitch_fx (Variables|B_Athena_PartModifier_DyeBlast|GetPlay%20Glitch%20FX))
  ;[0 0 0] Eliminations
  (|CallParentFunction State%20Name Stat%20Value)
  (Components|Timeline|SetPlayRate _play_glitch_fx (/ 1.0 (Variables|Default|GetGlitch%20Speed)))
  (Components|Timeline|PlayfromStart _play_glitch_fx))

(event Custom|FrontEnd%20Loop
  (bind _play_glitch_fx (Variables|B_Athena_PartModifier_DyeBlast|GetPlay%20Glitch%20FX))
  ;[0 0 0] Eliminations
  (Components|Timeline|SetPlayRate _play_glitch_fx (/ 1.0 (Variables|Default|GetGlitch%20Speed)))
  (Components|Timeline|PlayfromStart _play_glitch_fx))

(event EventOnDayPhaseChanged
  (params (CurrentDayPhase :type EFortDayPhase%20Enum :default "Morning") (PreviousDayPhase :type EFortDayPhase%20Enum :default "Morning") (bAtCreation :type Boolean)))

(event EventOnNewPlayerWeaponType
  (params (NewType :type EFortItemType%20Enum :default "WorldItem") (OldType :type EFortItemType%20Enum :default "WorldItem"))
  (|CallParentFunction NewType OldType) ; This node is disabled and will not be called.\nDrag off pins to build functionality.
)

(event EventOnWeaponChanged
  (params (NewWeapon :type Fort%20Weapon%20Object%20Reference) (OldWeapon :type Fort%20Weapon%20Object%20Reference)))

(event EventOnWeaponFired)

(event EventOnNewWeapons
  (params (WeaponsAsItems :type Array%20of%20Fort%20Item%20Object%20References)))

(event EventOnCurieStateAttached
  (params (Tag :type Gameplay%20Tag%20Structure%20%28by%20ref%29)))

(event EventOnCurieStateDetached
  (params (Tag :type Gameplay%20Tag%20Structure%20%28by%20ref%29)))

(event EventOnBeginSkydiving
  (|CallParentFunction) ; This node is disabled and will not be called.\nDrag off pins to build functionality.
)

(event EventOnEndSkydiving
  (|CallParentFunction) ; This node is disabled and will not be called.\nDrag off pins to build functionality.
)

(event EventOnEnterVehicle
  (params (VehicleEntered :type Fort%20Vehicle%20Interface%20Interface%20%28by%20ref%29))
  (|CallParentFunction VehicleEntered) ; This node is disabled and will not be called.\nDrag off pins to build functionality.
)

(event EventOnExitVehicle
  (|CallParentFunction) ; This node is disabled and will not be called.\nDrag off pins to build functionality.
)

(event EventOnTargetingOrFirstPersonCameraChanged)

(fn AudioTrigger
  (params (SoundLibTag :type Sound%20Library%20Tag%20Structure))
  (return false))

(fn Kills)

(fn SetTimestepFadeIn)

(fn SetTimestepPanner)

(fn SetTimestepPannerB)

(fn SetTimestepPannerC)

(fn ConstructionScript
  (|CallParentFunction))
