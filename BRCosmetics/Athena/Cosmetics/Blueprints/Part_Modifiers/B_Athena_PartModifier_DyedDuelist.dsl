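; Part modifier for the Dyed Duelist cosmetic that makes the character part react to kills.
; After res-in it caches material slot 0 as a dynamic instance and watches the AthenaKills stat.
; Each kill plays a timeline driving Fresnel, gem and sleeve-outline blends, and spawns an attached kill effect passed the hidden-hand flag.
; In frontend display, a looping 7.2-second timer previews the same reaction.
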(event Custom|FrontEnd_ReactivityPreview
  (bind _partmid (Variables|Default|GetPartMID))
  (bind _killfxsystem (Variables|Default|GetKillFXSystem))
  ;[1 1 1] Displays effect on frontend
  (if (Reactive|Frontend|IsCosmeticObjectDisplayedinFrontendContext 0)
    (Utilities|FlowControl|Sequence
      (:then_0
        (bind (_direction _boost) (Timeline|Timeline PlayFromStart)
          (:Update
            (Rendering|Material|SetScalarParameterValue _partmid "FresnelBlend" _boost :!class /Script/Engine.MaterialInstanceDynamic)
            (Rendering|Material|SetScalarParameterValue _partmid "GemBlend" _boost :!class /Script/Engine.MaterialInstanceDynamic)
            (Rendering|Material|SetScalarParameterValue _partmid "SleeveShapeOutlineBlend" _boost :!class /Script/Engine.MaterialInstanceDynamic))))
      (:then_1
        ;[1 1 1] NS Reactivity
        (Utilities|IsValid _killfxsystem
          (:Is%20Valid
            (bind _returnvalue (PartModifier|SpawnSystem|SpawnSystemAttachedPartModifier _killfxsystem (CustomCharacter|GetSkeletalMeshForPartOnPlayerPawn) "root")
              (:Success
                (Effects|Components|ParticleSystem|SetBoolParameter _returnvalue "bisHandHidden" (Variables|HiddenPartCheck|GetbIsHandHidden))))))))
    (else
      (Utilities|Time|ClearandInvalidateTimerbyHandle (Variables|Default|GetFrontEndReactivityTimer)))))

(event EventOnPostResIn
  (bind _partmid (Variables|Default|GetPartMID))
  (bind _killfxsystem (Variables|Default|GetKillFXSystem))
  (|CallParentFunction)
  (Utilities|FlowControl|Sequence
    (:then_1
      (bind _asmaterial_instance_dynamic (Utilities|Casting|CastToMaterialInstanceDynamic (Rendering|Material|GetMaterial (CustomCharacter|GetSkeletalMeshForPartOnPlayerPawn) 0)))
      (bind _output_get (Variables|Default|SetPartMID _asmaterial_instance_dynamic))
      (Utilities|IsValid _output_get
        (:Is%20Valid
          (Variables|Default|SetAssociatedPlayerPawn (CustomCharacter|GetAssociatedPlayerPawn))
          (Cosmetics|WatchCosmeticStat (Variables|Default|GetAssociatedPlayerPawn) "(TagName=\"StatManager.AthenaKills\")" -1 false
            (:OnStatChanged
              (Utilities|FlowControl|Sequence
                (:then_0
                  (Timeline|Timeline PlayFromStart))
                (:then_1
                  ;[1 1 1] NS Reactivity
                  (Utilities|IsValid _killfxsystem
                    (:Is%20Valid
                      (bind _returnvalue (PartModifier|SpawnSystem|SpawnSystemAttachedPartModifier _killfxsystem (CustomCharacter|GetSkeletalMeshForPartOnPlayerPawn) "root")
                        (:Success
                          (Effects|Components|ParticleSystem|SetBoolParameter _returnvalue "bisHandHidden" (Variables|HiddenPartCheck|GetbIsHandHidden)))))))))))))
    (:then_2
      ;[1 0.1636 0.04388] Frontend reactivity preview setup
      (if (Reactive|Frontend|IsCosmeticObjectDisplayedinFrontendContext 0)
        (bind _returnvalue_2 (Utilities|Time|SetTimerbyFunctionName 0 "FrontEnd_ReactivityPreview" 7.2 true))
        (Variables|Default|SetFrontEndReactivityTimer _returnvalue_2)
        (else
          (Utilities|Time|ClearandInvalidateTimerbyHandle (Variables|Default|GetFrontEndReactivityTimer)))))))

(event EventOnNewWeapons
  (params (WeaponsAsItems :type Array%20of%20Fort%20Item%20Object%20References)))

(event EventOnCurieStateAttached
  (params (Tag :type Gameplay%20Tag%20Structure%20%28by%20ref%29)))

(event EventOnDayPhaseChanged
  (params (CurrentDayPhase :type EFortDayPhase%20Enum :default "Morning") (PreviousDayPhase :type EFortDayPhase%20Enum :default "Morning") (bAtCreation :type Boolean)))

(event EventOnCurieStateDetached
  (params (Tag :type Gameplay%20Tag%20Structure%20%28by%20ref%29)))

(event EventOnNewPlayerWeaponType
  (params (NewType :type EFortItemType%20Enum :default "WorldItem") (OldType :type EFortItemType%20Enum :default "WorldItem"))
  (|CallParentFunction NewType OldType) ; This node is disabled and will not be called.\nDrag off pins to build functionality.
)

(event EventOnBeginSkydiving
  (|CallParentFunction) ; This node is disabled and will not be called.\nDrag off pins to build functionality.
)

(event EventOnWeaponChanged
  (params (NewWeapon :type Fort%20Weapon%20Object%20Reference) (OldWeapon :type Fort%20Weapon%20Object%20Reference)))

(event EventOnEndSkydiving
  (|CallParentFunction) ; This node is disabled and will not be called.\nDrag off pins to build functionality.
)

(event EventOnWeaponFired)

(event EventOnEnterVehicle
  (params (VehicleEntered :type Fort%20Vehicle%20Interface%20Interface%20%28by%20ref%29))
  (|CallParentFunction VehicleEntered) ; This node is disabled and will not be called.\nDrag off pins to build functionality.
)

(event EventOnExitVehicle
  (|CallParentFunction) ; This node is disabled and will not be called.\nDrag off pins to build functionality.
)

(event EventOnTargetingOrFirstPersonCameraChanged)

(event EventOnHitPawn
  (params (SelfActor :type Actor%20Object%20Reference) (OtherActor :type Actor%20Object%20Reference) (NormalImpulse :type Vector :default "0, 0, 0") (Hit :type Hit%20Result%20Structure%20%28by%20ref%29)))

(event Custom|OnKillSFX
  (Audio|SpawnSoundAttached (Variables|Default|GetSound) (CustomCharacter|GetSkeletalMeshForPartOnPlayerPawn)))

(event Custom|FX_OnKillStatChange
  (Utilities|FlowControl|Sequence
    (:then_0
      (bind _asniagara_particle_system_component (Utilities|Casting|CastToNiagaraComponent (CallFunction|GetIdleVFXComponent)))
      (Niagara|SetNiagaraVariable(Bool) _asniagara_particle_system_component "bKillReaction" true)
      (Utilities|FlowControl|RetriggerableDelay 3.0)
      (Niagara|SetNiagaraVariable(Bool) _asniagara_particle_system_component "bKillReaction" false))
    (:then_1
      (PartModifier|SpawnSystem|SpawnSystemAttachedPartModifier 0 (CustomCharacter|GetSkeletalMeshForPartOnPlayerPawn) "None" "0, 0, 0" "0, 0, 0" "SnapToTarget"))))

(fn ConstructionScript
  (|CallParentFunction))
