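; Part modifier for the EarthLane back bling that drives a reactive idle VFX, noted as triggering on eliminations.
; Each reaction writes current game time into the particle system's GameTime parameter, ignoring the event's boolean input.
; Positive stat changes trigger a reaction, plus an attached sound outside frontend reactivity display.
; After res-in it sets the CharacterPart parameter and, for frontend display, fires two delayed demo reactions.
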(event EventOnPostResIn
  ;[0 0 0] on post res-in
  (|CallParentFunction)
  (Effects|Components|ParticleSystem|SetFloatParameter (CallFunction|GetIdleVFXComponent) "CharacterPart" 1.0)
  (if (Reactive|Frontend|ShouldDisplayFrontendReactivity 0 "Backpack")
    (Utilities|FlowControl|Delay 0.5)
    (CallFunction|ReactivityEvent true)
    (Utilities|FlowControl|Delay 3.0)
    (CallFunction|ReactivityEvent true)))

(event EventOnStatChanged
  (params (State%20Name :type Name :default "None") (Stat%20Value :type Integer))
  ;[0 0 0] Trigger reactivity on Elimination (See Class Defaults For Tag Setup)
  (|CallParentFunction State%20Name Stat%20Value)
  (if (> Stat%20Value 0)
    (CallFunction|ReactivityEvent true)
    (if (not (Reactive|Frontend|ShouldDisplayFrontendReactivity 0 "Backpack"))
      (Audio|SpawnSoundAttached "/BRCosmetics/Sounds/Backblings/EarthLane/Backpack_EarthLane_Cue.Backpack_EarthLane_Cue" (CustomCharacter|GetSkeletalMeshForPartOnPlayerPawn)))))

(event Custom|ReactivityEvent
  (params (isReactive? :type Boolean))
  ;[0 0 0] ReactivityEvent
  (Effects|Components|ParticleSystem|SetFloatParameter (CallFunction|GetIdleVFXComponent) "GameTime" (Utilities|Time|GetGameTimeinSeconds)))

(fn ConstructionScript
  (|CallParentFunction))
