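; Cosmetic part modifier for the Ebony back bling that makes the backpack react to its wearer's kill count.
; When the pawn or part changes, it caches the current Athena kills stat in KillCounterClamped and watches that stat.
; On each change it clamps the count to a small range and calls KillCountClampedUpdated on the backpack's anim blueprint.
; The initial read is neither clamped nor sent to the anim blueprint.
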(event EventOnPawnAndPartChanged
  (bind _returnvalue (CustomCharacter|GetAssociatedPlayerPawn))
  (bind _returnvalue_2 (Customization|GetSkeletalMeshforPartType _returnvalue "Backpack" :!class /Script/FortniteGame.FortPlayerPawn))
  (bind _returnvalue_3 (Components|SkeletalMesh|GetAnimInstance _returnvalue_2))
  (|CallParentFunction)
  (bind _returnvalue_4 (Class|FortCosmeticFunctionLibrary|GetCosmeticStatValue _returnvalue "(TagName=\"StatManager.AthenaKills\")"))
  (Variables|Default|SetKillCounterClamped _returnvalue_4)
  (bind (_statname _statvalue) (Cosmetics|WatchCosmeticStat _returnvalue "(TagName=\"StatManager.AthenaKills\")" 0 false)
    (:OnStatChanged
      (Variables|Default|SetKillCounterClamped (Math|Integer|Clamp(Integer) _statvalue 0 6))
      (Utilities|IsValid _returnvalue_3
        (:Is%20Valid
          (bind _asm_med_ebony_pack_anim_bp (Utilities|Casting|CastToM_MED_Ebony_Pack_AnimBP _returnvalue_3))
          (Class|MMEDEbonyPackAnimBP|KillCountClampedUpdated _asm_med_ebony_pack_anim_bp))))))

(event EventOnPostResIn
  (|CallParentFunction))

(event EventOnDayPhaseChanged
  (params (CurrentDayPhase :type EFortDayPhase%20Enum :default "Morning") (PreviousDayPhase :type EFortDayPhase%20Enum :default "Morning") (bAtCreation :type Boolean)))

(event EventOnNewPlayerWeaponType
  (params (NewType :type EFortItemType%20Enum :default "WorldItem") (OldType :type EFortItemType%20Enum :default "WorldItem"))
  (|CallParentFunction NewType OldType) ; This node is disabled and will not be called.\nDrag off pins to build functionality.
)

(event EventOnWeaponChanged
  (params (NewWeapon :type Fort%20Weapon%20Object%20Reference) (OldWeapon :type Fort%20Weapon%20Object%20Reference)))

(event EventOnWeaponFired)

(event EventOnNewWeapons
  (params (WeaponsAsItems :type Array%20of%20Fort%20Item%20Object%20References)))

(event EventOnCurieStateAttached
  (params (Tag :type Gameplay%20Tag%20Structure%20%28by%20ref%29)))

(event EventOnCurieStateDetached
  (params (Tag :type Gameplay%20Tag%20Structure%20%28by%20ref%29)))

(event EventOnBeginSkydiving
  (|CallParentFunction) ; This node is disabled and will not be called.\nDrag off pins to build functionality.
)

(event EventOnEndSkydiving
  (|CallParentFunction) ; This node is disabled and will not be called.\nDrag off pins to build functionality.
)

(event EventOnEnterVehicle
  (params (VehicleEntered :type Fort%20Vehicle%20Interface%20Interface%20%28by%20ref%29))
  (|CallParentFunction VehicleEntered) ; This node is disabled and will not be called.\nDrag off pins to build functionality.
)

(event EventOnExitVehicle
  (|CallParentFunction) ; This node is disabled and will not be called.\nDrag off pins to build functionality.
)

(event EventOnTargetingOrFirstPersonCameraChanged)

(event EventOnHitPawn
  (params (SelfActor :type Actor%20Object%20Reference) (OtherActor :type Actor%20Object%20Reference) (NormalImpulse :type Vector :default "0, 0, 0") (Hit :type Hit%20Result%20Structure%20%28by%20ref%29)))

(fn ConstructionScript
  (|CallParentFunction))
