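; Backpack part modifier for the Elastic cosmetic that drives a hologram texture on the backpack's idle Niagara VFX.
; When the character belongs to the Endless set and reactivity is allowed, it copies the body material's Sticker texture into a dynamic hologram material for Niagara.
; A fallback path applies backpack variants and attaches the VFX to the part mesh.
; Textures over 256 pixels on both axes count as animated sprite sheets, toggling AnimatedHologram on material and Niagara.
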(event EventOnPostResIn
  (|CallParentFunction)
  (CallFunction|SetupTexture))

(event Custom|Setup%20Texture
  (bind _returnvalue (CallFunction|GetIdleVFXComponent))
  (bind _asniagara_particle_system_component (Utilities|Casting|CastToNiagaraComponent _returnvalue))
  (Utilities|ShouldAllowElasticReactive
    (:Yes
      (if (CallFunction|IsCharacterElastic)
        (bind _result (ArrayMacros|SafeArrayGet (Rendering|Material|GetMaterials (Customization|GetSkeletalMeshforPartType (CustomCharacter|GetAssociatedPlayerPawn) "Body" :!class /Script/FortniteGame.FortPlayerPawn)))
          (:%20%20
            (bind _asmaterial_instance_dynamic (Utilities|Casting|CastToMaterialInstanceDynamic _result)
              (:then
                (bind _returnvalue_2 (Rendering|Material|GetTextureParameterValue _asmaterial_instance_dynamic "Sticker" :!class /Script/Engine.MaterialInstanceDynamic))
                (bind _returnvalue_3 (Rendering|Material|CreateDynamicMaterialInstance "/BRCosmetics/Characters/Player/Female/Medium/Bodies/F_MED_Elastic/FX/M_Elastic_Holograms.M_Elastic_Holograms" "Elastic Hologram MID" "None" :!class /Script/Engine.KismetMaterialLibrary))
                (Rendering|Material|SetTextureParameterValue _returnvalue_3 "2DHologramTexture" _returnvalue_2)
                (Niagara|SetNiagaraVariable(Material) _asniagara_particle_system_component "MID_ElasticHologram" _returnvalue_3)
                (CallFunction|SetupAnimatedMaterial))
              (:CastFailed
                (Development|PrintString "CCPM_Elastic_MID Cast FAILED")))))
        (else
          (Utilities|GateBackpackVariantFallback))))
    (:No
      _))
  ;[1 1 1] First we need to make sure that we actually have a dynamic material...
  (bind _returnvalue_4 (Rendering|Material|CreateDynamicMaterialInstance "/BRCosmetics/Characters/Player/Female/Medium/Bodies/F_MED_Elastic/FX/M_Elastic_Holograms.M_Elastic_Holograms" "Elastic Hologram MID" "None" :!class /Script/Engine.KismetMaterialLibrary))
  (bind _output_get (Variables|Default|SetInstanceMaterial _returnvalue_4))
  (Niagara|SetNiagaraVariable(Material) _asniagara_particle_system_component "MID_ElasticHologram" _output_get)
  ;[1 1 1] Now, we can have the backpack handle applying the variant.
  (Class|FortCosmeticFunctionLibrary|ApplyVariantstoComponentbySlotTag (CustomCharacter|GetAssociatedPlayerPawn) "(TagName=\"LoadoutSlot.Backpack\")" (CallFunction|GetIdleVFXComponent)
    (:True
      _)
    (:False
      _))
  (Transformation|AttachComponentToComponent (CallFunction|GetIdleVFXComponent) (CustomCharacter|GetSkeletalMeshForPartOnPlayerPawn) (CallFunction|GetIdleFXSocketName) "SnapToTarget" "SnapToTarget" "SnapToTarget" true)
  (CallFunction|SetupAnimatedMaterial)
  (bind _returnvalue_5 (Rendering|Material|GetTextureParameterValue (Variables|Default|GetInstanceMaterial) "2DHologramTexture" :!class /Script/Engine.MaterialInstanceDynamic))
  (Niagara|SetNiagaraVariable(Texture) _asniagara_particle_system_component "HologramTexture" _returnvalue_5))

(fn GrabLoadout
  (return (Variables|/Script/FortniteGame.FortPlayerPawn|GetCosmeticLoadout (CustomCharacter|GetAssociatedPlayerPawn))))

(fn IsCharacterElastic
  (return (GameplayTags|HasTag (GameplayTags|GetOwnedGameplayTags (CosmeticLoadout|GetItemDefinitionfromLoadout (CosmeticLoadout|GetCosmeticLoadout) "(TagName=\"LoadoutSlot.Character\")") :!class /Script/GameplayTags.GameplayTagAssetInterface) "(TagName=\"Cosmetics.Set.Endless\")")))

(fn SetupAnimatedMaterial
  (bind _returnvalue (CallFunction|GetIdleVFXComponent))
  (bind _asniagara_particle_system_component (Utilities|Casting|CastToNiagaraComponent _returnvalue))
  (bind _instancematerial (Variables|Default|GetInstanceMaterial))
  (Utilities|IsValid _instancematerial
    (:Is%20Valid
      (bind _returnvalue_2 (Rendering|Material|GetTextureParameterValue _instancematerial "2DHologramTexture" :!class /Script/Engine.MaterialInstanceDynamic))
      ;[1 1 1] Check Texture Size to see if texure is a sprite sheet or not
      (bind _size (Class|VolumetricsBPLIbrary|GetTextureSize _returnvalue_2))
      (if (and (> (Utilities|Struct|BreakIntPoint _size) 256) (> (.Y (Utilities|Struct|BreakIntPoint _size)) 256))
        (Rendering|Material|SetScalarParameterValue _instancematerial "AnimatedHologram" 1.0 :!class /Script/Engine.MaterialInstanceDynamic)
        (Variables|Default|SetIsAnimated true)
        (Niagara|SetNiagaraVariable(Float) _asniagara_particle_system_component "AnimatedHologram" 1.0)
        (else
          (Rendering|Material|SetScalarParameterValue _instancematerial "AnimatedHologram" 0.0 :!class /Script/Engine.MaterialInstanceDynamic)
          (Variables|Default|SetIsAnimated false)
          (Niagara|SetNiagaraVariable(Float) _asniagara_particle_system_component "AnimatedHologram" 0.0))))))

(fn ConstructionScript
  (|CallParentFunction))
