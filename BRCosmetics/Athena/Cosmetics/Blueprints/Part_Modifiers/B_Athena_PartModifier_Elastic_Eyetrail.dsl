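; Character part modifier for the Elastic eye trail cosmetic, adapting its idle VFX to time of day and aiming.
; Day phase changes set the eye glow and trail multiplier parameters on the idle VFX component, with a stronger setting for Day.
; Targeting or first-person camera changes call the parent, cache the pawn's targeting state and pass it to the effect's isADS parameter only for locally controlled pawns.
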(event EventOnDayPhaseChanged
  (params (CurrentDayPhase :type EFortDayPhase%20Enum :default "Morning") (PreviousDayPhase :type EFortDayPhase%20Enum :default "Morning") (bAtCreation :type Boolean))
  (bind _returnvalue (CallFunction|GetIdleVFXComponent))
  (bind _returnvalue_2 (CallFunction|GetIdleVFXComponent))
  ;[1 1 1 1 30] Controls the emissive values depending on Time of Day
  (switch Utilities|FlowControl|Switch|SwitchonEFortDayPhase CurrentDayPhase
    (:Morning
      _)
    (:Day
      (Effects|Components|ParticleSystem|SetFloatParameter _returnvalue "EyeGlowMultiplier" 0.35)
      (Effects|Components|ParticleSystem|SetFloatParameter _returnvalue "EyeTrailMultiplier" 1.0))
    (:Evening
      _)
    (:Night
      _))
  (Effects|Components|ParticleSystem|SetFloatParameter _returnvalue_2 "EyeGlowMultiplier" 0.1)
  (Effects|Components|ParticleSystem|SetFloatParameter _returnvalue_2 "EyeTrailMultiplier" 0.7))

(event EventOnTargetingOrFirstPersonCameraChanged
  (bind _returnvalue (CustomCharacter|GetAssociatedPlayerPawn))
  (|CallParentFunction)
  (Variables|Default|SetbIsADSing (Variables|/Script/FortniteGame.FortPlayerPawn|GetbIsTargeting _returnvalue))
  (if (Pawn|IsLocallyControlled _returnvalue)
    (Effects|Components|ParticleSystem|SetBoolParameter (CallFunction|GetIdleVFXComponent) "isADS" (Variables|Default|GetbIsADSing))))

(fn ConstructionScript)
