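; Cosmetic part modifier for the EmeraldGlassStandAlone backpack that shows a Niagara effect while the player skydives.
; On skydive start it checks the part's skeletal mesh, with a short delayed recheck if invalid, and spawns the backpack's idle system attached at FX_Steam.
; It stores the spawned component in SkydiveFX and deactivates it when skydiving ends, if still valid.
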(event EventOnBeginSkydiving
  (bind _returnvalue (CustomCharacter|GetSkeletalMeshForPartOnPlayerPawn))
  (|CallParentFunction)
  (Utilities|FlowControl|Sequence
    (:then_1
      (Utilities|IsValid _returnvalue
        (:Is%20Valid
          (Utilities|FlowControl|Sequence
            (:then_0
              _)))
        (:Is%20Not%20Valid
          (Utilities|FlowControl|Delay 0.1)
          (Utilities|IsValid (CustomCharacter|GetSkeletalMeshForPartOnPlayerPawn)
            (:Is%20Valid
              _))))
      (bind _returnvalue_2 (PartModifier|SpawnSystem|SpawnSystemAttachedPartModifier "/BRCosmetics/Accessories/FORT_Backpacks/Backpack_M_MED_EmeraldGlassStandAlone/FX/NS_EmeraldGlassStandAlone_Backpack_Idle.NS_EmeraldGlassStandAlone_Backpack_Idle" _returnvalue "FX_Steam")
        (:Success
          (Variables|Default|SetSkydiveFX _returnvalue_2))))))

(event EventOnEndSkydiving
  (bind _skydivefx (Variables|Default|GetSkydiveFX))
  (|CallParentFunction)
  (Utilities|IsValid _skydivefx
    (:Is%20Valid
      (Components|Activation|Deactivate _skydivefx))))

(fn ConstructionScript
  (|CallParentFunction))
