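; Character part modifier for the EssayViewPier cosmetic that drives body-material reactivity and idle VFX in matches and in the frontend.
; On post-res-in it caches the player pawn and sets the body material's IsInFrontend flag.
; In the frontend, a looping timer re-activates the idle VFX while a timeline drives ReactivityAlpha.
; Positive stat changes stamp ReactivityStartTime with game time and activate the VFX through a DoOnce.
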(event EventOnStatChanged
  (params (State%20Name :type Name :default "None") (Stat%20Value :type Integer))
  (|CallParentFunction State%20Name Stat%20Value)
  (if (> Stat%20Value 0)
    (Rendering|Material|SetScalarParameterValueonMaterials (Customization|GetSkeletalMeshforPartType (Variables|Default|GetPlayer%20Pawn) "Body" :!class /Script/FortniteGame.FortPlayerPawn) "ReactivityStartTime" (Utilities|Time|GetGameTimeinSeconds))
    (Utilities|FlowControl|DoOnce false
      (:Completed
        (Components|Activation|Activate (CallFunction|GetIdleVFXComponent) true)
        (Utilities|FlowControl|RetriggerableDelay 1.5)))))

(event EventOnPostResIn
  (bind _player_pawn (Variables|Default|GetPlayer%20Pawn))
  (bind _returnvalue (Customization|GetSkeletalMeshforPartType _player_pawn "Body" :!class /Script/FortniteGame.FortPlayerPawn))
  (|CallParentFunction)
  (Variables|Default|SetPlayer%20Pawn (CustomCharacter|GetAssociatedPlayerPawn))
  (if (CustomCosmeticModifier|IsinFrontEnd)
    (Rendering|Material|SetScalarParameterValueonMaterials _returnvalue "IsInFrontend" 1.0)
    (Utilities|FlowControl|Sequence
      (:then_0
        (Networking|CanExecuteCosmeticEvents
          (:True
            (bind _returnvalue_2 (Utilities|Time|SetTimerbyEvent (delegate Custom|CE%20FN%20Loop) 1e+01 true false -9.0))
            (Variables|Default|SetTimer _returnvalue_2))))
      (:then_1
        (bind (_direction _alpha) (Timeline|ReactivityFrontend Play)
          (:Update
            (Rendering|Material|SetScalarParameterValueonMaterials _returnvalue "ReactivityAlpha" _alpha)))))
    (else
      (Rendering|Material|SetScalarParameterValueonMaterials _returnvalue "IsInFrontend" 0.0))))

(event Custom|CE%20FN%20Loop
  (Components|Activation|Activate (CallFunction|GetIdleVFXComponent) true))

(event EventDestroyed
  (|CallParentFunction)
  (if (CustomCosmeticModifier|IsinFrontEnd)
    (Utilities|Time|ClearandInvalidateTimerbyHandle (Variables|Default|GetTimer))))

(fn ConstructionScript
  (|CallParentFunction))
