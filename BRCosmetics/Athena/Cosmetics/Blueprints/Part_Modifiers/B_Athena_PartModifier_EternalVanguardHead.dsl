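; Part modifier Blueprint for the Eternal Vanguard head cosmetic in Athena.
; When the post-res-in event fires, it calls the parent logic, then sets the translucent sort priority of the idle VFX component to -1, lowering its draw order among translucent objects.
; The construction script only calls the parent implementation.
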(event EventOnPostResIn
  (|CallParentFunction)
  (Rendering|SetTranslucentSortPriority (CallFunction|GetIdleVFXComponent) -1))

(fn ConstructionScript
  (|CallParentFunction))
