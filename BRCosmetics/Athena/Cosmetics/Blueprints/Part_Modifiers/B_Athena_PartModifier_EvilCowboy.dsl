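; Evil Cowboy character part modifier that switches the outfit between three cosmetic stages as cosmetic state activates, with a timed stage preview in the front end.
; Each stage applies color and scalar parameter maps to the part's mesh materials.
; Stages 1 and 2 also spawn variant-aware Niagara transition effects and respectively stop or start the idle VFX.
; Skydiving disables the idle effect's upper flame emitter until it ends.
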(event Custom|Stage0
  (bind _returnvalue (CustomCharacter|GetAssociatedPlayerPawn))
  (bind _returnvalue_2 (CustomCharacter|GetCharacterPartType))
  (bind _returnvalue_3 (Customization|GetSkeletalMeshforPartType _returnvalue _returnvalue_2 :!class /Script/FortniteGame.FortPlayerPawn))
  (bind _stage_0_material_colors (Variables|Default|GetStage%200%20Material%20Colors))
  (bind _stage_0_material_scalars (Variables|Default|GetStage%200%20Material%20Scalars))
  (Networking|CanExecuteCosmeticEvents
    (:True
      (Utilities|FlowControl|Sequence
        (:then_0
          ;[1 1 1] Stage 0 Material Changes for Character Part - Matches Character Defaults
          (bind _keys (Utilities|Map|Keys _stage_0_material_colors))
          (if (CallFunction|ArrayIsValid _keys)
            (for _array_element _keys
              (if (Utilities|Name|Equal(Name) _array_element "None")
                (Development|PrintString "What are you thinking?  A Variable called None.  Really?!")
                (else
                  (Rendering|Material|SetVectorParameterValueonMaterials _returnvalue_3 _array_element (Math|Conversions|ToVector(LinearColor) (Utilities|Map|Find _stage_0_material_colors _array_element))))))
            (bind _keys_2 (Utilities|Map|Keys _stage_0_material_scalars))
            (if (CallFunction|ArrayIsValid _keys_2)
              (for _array_element_2 _keys_2
                (if (Utilities|Name|Equal(Name) _array_element_2 "None")
                  (Development|PrintString "What are you thinking?  A Variable called None.  Really?!")
                  (else
                    (Rendering|Material|SetScalarParameterValueonMaterials _returnvalue_3 _array_element_2 (Utilities|Map|Find _stage_0_material_scalars _array_element_2))))))))))))

(event Custom|Stage%201
  (bind _returnvalue (CustomCharacter|GetAssociatedPlayerPawn))
  (bind _returnvalue_2 (CustomCharacter|GetCharacterPartType))
  (bind _returnvalue_3 (Customization|GetSkeletalMeshforPartType _returnvalue _returnvalue_2 :!class /Script/FortniteGame.FortPlayerPawn))
  (bind _stage_1_material_colors (Variables|Default|GetStage%201%20Material%20Colors))
  (bind _stage_1_material_scalars (Variables|Default|GetStage%201%20Material%20Scalars))
  (Networking|CanExecuteCosmeticEvents
    (:True
      (Utilities|FlowControl|Sequence
        (:then_0
          ;[1 1 1] Stage 1 Material Changes for Character Part
          (bind _keys (Utilities|Map|Keys _stage_1_material_colors))
          (if (CallFunction|ArrayIsValid _keys)
            (for _array_element _keys
              (if (Utilities|Name|Equal(Name) _array_element "None")
                (Development|PrintString "What are you thinking?  A Variable called None.  Really?!")
                (else
                  (Rendering|Material|SetVectorParameterValueonMaterials _returnvalue_3 _array_element (Math|Conversions|ToVector(LinearColor) (Utilities|Map|Find _stage_1_material_colors _array_element))))))
            (bind _keys_2 (Utilities|Map|Keys _stage_1_material_scalars))
            (if (CallFunction|ArrayIsValid _keys_2)
              (for _array_element_2 _keys_2
                (if (Utilities|Name|Equal(Name) _array_element_2 "None")
                  (Development|PrintString "What are you thinking?  A Variable called None.  Really?!")
                  (else
                    (Rendering|Material|SetScalarParameterValueonMaterials _returnvalue_3 _array_element_2 (Utilities|Map|Find _stage_1_material_scalars _array_element_2))))))))
        (:then_1
          ;[1 1 1] Stage 1 VFX
          (bind _returnvalue_4 (PartModifier|SpawnSystem|SpawnSystemAttachedPartModifier "/BRCosmetics/Athena/Effects/Particles/Skins/EvilCowboy/NS_Character_EvilCowboy_Stage02_OnBodyTransition.NS_Character_EvilCowboy_Stage02_OnBodyTransition" (CustomCharacter|GetSkeletalMeshForPartOnPlayerPawn) "None" "0, 0, 0" "0, 0, 0" "SnapToTarget")
            (:Success
              (Cosmetics|ApplyVariantsToComponent(Deprecated) :self (Variables|Default|GetCharacter%20Item%20Definition%20Reference) :InComponent _returnvalue_4 :Loadout (Variables|/Script/FortniteGame.FortPlayerPawn|GetCosmeticLoadout (CustomCharacter|GetAssociatedPlayerPawn)))
              (Components|Activation|Activate _returnvalue_4 false)
              (if (Components|Activation|IsActive (CallFunction|GetIdleVFXComponent))
                (Components|Activation|Deactivate (CallFunction|GetIdleVFXComponent))))))))))

(event EventOnBeginSkydiving
  (bind _returnvalue (CallFunction|GetIdleVFXComponent))
  (if (Components|Activation|IsActive _returnvalue)
    (Effects|Components|ParticleSystem|SetEmitterEnable _returnvalue "Flames_UpperSpeed" false)))

(event EventOnEndSkydiving
  (bind _returnvalue (CallFunction|GetIdleVFXComponent))
  (if (Components|Activation|IsActive _returnvalue)
    (Effects|Components|ParticleSystem|SetEmitterEnable _returnvalue "Flames_UpperSpeed" true)))

(event EventOnPostResIn
  (|CallParentFunction)
  (if (CustomCosmeticModifier|IsinFrontEnd)
    (Utilities|FlowControl|Delay 2.0)
    (Utilities|FlowControl|Sequence
      (:then_0
        (Utilities|FlowControl|Delay 3.0)
        (CallFunction|Stage2))
      (:then_1
        (CallFunction|Stage1)))))

(event EventOnCosmeticStateActivated
  (params (Stat%20Name :type Name :default "None") (Stat%20Value :type Integer) (STAGE%20Num%20Being%20Activated :type Integer))
  (|CallParentFunction Stat%20Name Stat%20Value STAGE%20Num%20Being%20Activated)
  (switch int STAGE%20Num%20Being%20Activated
    (:0
      (CallFunction|Stage0))
    (:1
      (CallFunction|Stage1))
    (:2
      (CallFunction|Stage2))))

(event Custom|Stage%202
  (bind _returnvalue (CustomCharacter|GetAssociatedPlayerPawn))
  (bind _returnvalue_2 (CustomCharacter|GetCharacterPartType))
  (bind _returnvalue_3 (Customization|GetSkeletalMeshforPartType _returnvalue _returnvalue_2 :!class /Script/FortniteGame.FortPlayerPawn))
  (bind _stage_2_material_colors (Variables|Default|GetStage%202%20Material%20Colors))
  (bind _stage_2_material_scalars (Variables|Default|GetStage%202%20Material%20Scalars))
  (Networking|CanExecuteCosmeticEvents
    (:True
      (Utilities|FlowControl|Sequence
        (:then_0
          ;[1 1 1] Stage 1 Material Changes for Character Part
          (bind _keys (Utilities|Map|Keys _stage_2_material_colors))
          (if (CallFunction|ArrayIsValid _keys)
            (for _array_element _keys
              (if (Utilities|Name|Equal(Name) _array_element "None")
                (Development|PrintString "What are you thinking?  A Variable called None.  Really?!")
                (else
                  (Rendering|Material|SetVectorParameterValueonMaterials _returnvalue_3 _array_element (Math|Conversions|ToVector(LinearColor) (Utilities|Map|Find _stage_2_material_colors _array_element))))))
            (bind _keys_2 (Utilities|Map|Keys _stage_2_material_scalars))
            (if (CallFunction|ArrayIsValid _keys_2)
              (for _array_element_2 _keys_2
                (if (Utilities|Name|Equal(Name) _array_element_2 "None")
                  (Development|PrintString "What are you thinking?  A Variable called None.  Really?!")
                  (else
                    (Rendering|Material|SetScalarParameterValueonMaterials _returnvalue_3 _array_element_2 (Utilities|Map|Find _stage_2_material_scalars _array_element_2))))))))
        (:then_1
          ;[1 1 1] Stage 2 VFX
          (bind _returnvalue_4 (PartModifier|SpawnSystem|SpawnSystemAttachedPartModifier "/BRCosmetics/Athena/Effects/Particles/Skins/EvilCowboy/NS_Character_EvilCowboy_Stage03_OnBodyTransition.NS_Character_EvilCowboy_Stage03_OnBodyTransition" (CustomCharacter|GetSkeletalMeshForPartOnPlayerPawn) "None" "0, 0, 0" "0, 0, 0" "SnapToTarget")
            (:Success
              (Cosmetics|ApplyVariantsToComponent(Deprecated) :self (Variables|Default|GetCharacter%20Item%20Definition%20Reference) :InComponent _returnvalue_4 :Loadout (Variables|/Script/FortniteGame.FortPlayerPawn|GetCosmeticLoadout (CustomCharacter|GetAssociatedPlayerPawn)))
              (Components|Activation|Activate _returnvalue_4 false))
            (:Failure
              _))
          (Components|Activation|Activate (CallFunction|GetIdleVFXComponent) true))))))

(event Custom|FE_Changer
  (switch int (Variables|Default|GetFE_Selection)
    (:0
      (CallFunction|Stage0))
    (:1
      (CallFunction|Stage1))
    (:2
      (CallFunction|Stage2)
      (Variables|Default|SetFE_Selection 0)))
  (Variables|Default|SetFE_Selection (+ (Variables|Default|GetFE_Selection) 1)))

(fn ArrayIsValid
  (params (Name%20Array :type Array%20of%20Names))
  (return (> (Utilities|Array|Length Name%20Array) 0)))

(fn ConstructionScript
  (|CallParentFunction))
