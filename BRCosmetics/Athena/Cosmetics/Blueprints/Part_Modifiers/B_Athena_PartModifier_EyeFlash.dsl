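; B_Athena_PartModifier_EyeFlash is a cosmetic character part modifier that spawns an eye flash Niagara effect in response to stat changes.
; When a changed stat's value is above zero, it spawns its first system attached to the root of the part's skeletal mesh on the player pawn.
; On a successful spawn, it passes its EffectColor and EffectPositionOffset variables into the Niagara component to tint and position the flash.
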(event EventOnStatChanged
  (params (State%20Name :type Name :default "None") (Stat%20Value :type Integer))
  (|CallParentFunction State%20Name Stat%20Value)
  (if (> Stat%20Value 0)
    (bind _returnvalue (PartModifier|SpawnSystem|SpawnSystemAttachedPartModifier 0 (CustomCharacter|GetSkeletalMeshForPartOnPlayerPawn) "root" "0, 0, 0" "0, 0, 0" "SnapToTarget")
      (:Success
        (Niagara|SetNiagaraVariable(LinearColor) _returnvalue "EffectColor" (Variables|Default|GetEffectColor))
        (Niagara|SetNiagaraVariable(Vector3) _returnvalue "EffectPositionOffset" (Variables|Default|GetEffectPositionOffset))))))

(fn ConstructionScript
  (|CallParentFunction))
