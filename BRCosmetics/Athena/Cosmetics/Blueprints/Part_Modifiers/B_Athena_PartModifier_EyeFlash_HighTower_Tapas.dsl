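; Cosmetic part modifier for the HighTower Tapas character that plays an eye flash effect when a tracked stat changes to a value above zero.
; After calling the parent stat handler, it spawns the HighTower Honeydew eye Niagara system attached at the root of the character part's skeletal mesh.
; On a successful spawn, it passes the effect color, position offset and head-hidden flag from its variables into the Niagara system.
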(event EventOnStatChanged
  (params (State%20Name :type Name :default "None") (Stat%20Value :type Integer))
  (|CallParentFunction State%20Name Stat%20Value)
  (if (> Stat%20Value 0)
    (bind _returnvalue (PartModifier|SpawnSystem|SpawnSystemAttachedPartModifier "/BRCosmetics/Characters/Player/Female/Medium/Bodies/F_MED_HighTower_Honeydew/HighTower_GeneralFX/NS_EyeEffect.NS_EyeEffect" (CustomCharacter|GetSkeletalMeshForPartOnPlayerPawn) "root" "0, 0, 0" "0, 0, 0" "SnapToTarget")
      (:Success
        (Niagara|SetNiagaraVariable(LinearColor) _returnvalue "EffectColor" (Variables|Default|GetEffectColor))
        (Niagara|SetNiagaraVariable(Vector3) _returnvalue "EffectPositionOffset" (Variables|Default|GetEffectPositionOffset))
        (Niagara|SetNiagaraVariable(Bool) _returnvalue "bisHeadHidden" (Variables|HiddenPartCheck|GetbIsHeadHidden))))))

(fn ConstructionScript
  (|CallParentFunction))
