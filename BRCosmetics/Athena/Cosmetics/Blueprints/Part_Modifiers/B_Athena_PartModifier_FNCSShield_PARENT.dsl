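; Parent cosmetic part modifier for the FNCS shield, judging by its name a back bling, that displays the player's kill count.
; After res-in it caches the part's skeletal mesh and watches the AthenaKills stat on the associated pawn, clamping it to 0–99.
; Each change immediately stamps game time and transition length into material scalars to drive a transition.
; A retriggerable delay then writes the Score parameter, so rapid kills share one counter update.
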(event EventOnPostResIn
  (|CallParentFunction)
  (Variables|Default|SetBackpackSkelMesh (CustomCharacter|GetSkeletalMeshForPartOnPlayerPawn))
  (bind (_statname _statvalue) (Cosmetics|WatchCosmeticStat (CustomCharacter|GetAssociatedPlayerPawn) "(TagName=\"StatManager.AthenaKills\")" -1)
    (:OnStatChanged
      (Variables|Default|SetKillCount (Math|Integer|Clamp(Integer) _statvalue 0 99))
      (Utilities|FlowControl|Sequence
        (:then_0
          (Rendering|Material|SetScalarParameterValueonMaterials (Variables|Default|GetBackpackSkelMesh) "Timestamp" (Utilities|Time|GetGameTimeinSeconds))
          (Rendering|Material|SetScalarParameterValueonMaterials (Variables|Default|GetBackpackSkelMesh) "TransitionLength" (Variables|Default|GetTransition%20Length)))
        (:then_1
          (Utilities|FlowControl|RetriggerableDelay (Variables|Default|GetDelayBeforeUpdatingCounterInt))
          (Rendering|Material|SetScalarParameterValueonMaterials (Variables|Default|GetBackpackSkelMesh) "Score" (Math|Conversions|ToFloat(Integer) (Variables|Default|GetKillCount))))))))

(fn ConstructionScript
  (|CallParentFunction))
