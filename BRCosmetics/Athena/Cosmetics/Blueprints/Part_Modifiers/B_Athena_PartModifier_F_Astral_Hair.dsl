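; Character part modifier for the female Astral cosmetic's hair, driving its idle VFX and materials from pawn and animation state.
; After post-res-in, it flags capes or long back bling on both; each tick it feeds LOD and rigid-body state to the VFX.
; Smoothed ground speed drives a running material parameter, only above LOD 0 or with back bling.
; Front-end holograms get an invisible-hair material on slot 1.
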(event EventOnPostResIn
  (|CallParentFunction)
  (if (CallFunction|IsonFire)
    (bind _ascustom_character_part_anim_instance (Utilities|Casting|CastToCustomCharacterPartAnimInstance (Components|SkeletalMesh|GetAnimInstance (Customization|GetSkeletalMeshforPartType (CustomCharacter|GetAssociatedPlayerPawn) (CustomCharacter|GetCharacterPartType) :!class /Script/FortniteGame.FortPlayerPawn))))
    (Variables|Default|SetHasBackbling (CustomCharacter|CheckforBackbling _ascustom_character_part_anim_instance 5e+01))
    (Effects|Components|ParticleSystem|SetBoolParameter (CallFunction|GetIdleVFXComponent) "HasCape" (or (.bIsBackblingACape (CustomCharacter|CheckforBackbling _ascustom_character_part_anim_instance 5e+01)) (.bIsBackblingLong (CustomCharacter|CheckforBackbling _ascustom_character_part_anim_instance 5e+01))))
    (Rendering|Material|SetScalarParameterValueonMaterials (Customization|GetSkeletalMeshforPartType (CustomCharacter|GetAssociatedPlayerPawn) (CustomCharacter|GetCharacterPartType) :!class /Script/FortniteGame.FortPlayerPawn) "HasLongCapeOrLongBP" (Math|Conversions|ToFloat(Boolean) (or (.bIsBackblingACape (CustomCharacter|CheckforBackbling _ascustom_character_part_anim_instance 5e+01)) (.bIsBackblingLong (CustomCharacter|CheckforBackbling _ascustom_character_part_anim_instance 5e+01)))))))

(event EventTick
  (params (DeltaSeconds :type Float%20%28single-precision%29))
  (bind _returnvalue (CustomCharacter|GetAssociatedPlayerPawn))
  (bind _returnvalue_2 (CustomCharacter|GetCharacterPartType))
  (bind _returnvalue_3 (Customization|GetSkeletalMeshforPartType _returnvalue _returnvalue_2 :!class /Script/FortniteGame.FortPlayerPawn))
  (CallFunction|Update DeltaSeconds)
  (if (CustomCosmeticModifier|IsinFrontEnd)
    (if (Variables|/Script/FortniteGame.FortPlayerPawn|GetbIsInFrontEndHologram (CustomCharacter|GetAssociatedPlayerPawn))
      (if (not (== (Rendering|Material|GetMaterial _returnvalue_3 1) "/BRCosmetics/Characters/Player/Female/Medium/Bodies/F_MED_Astral/FX/M_FX_Astral_InvisHair.M_FX_Astral_InvisHair"))
        (Rendering|Material|SetMaterial _returnvalue_3 1 "/BRCosmetics/Characters/Player/Female/Medium/Bodies/F_MED_Astral/FX/M_FX_Astral_InvisHair.M_FX_Astral_InvisHair")))))

(fn IsOnFire
  (return true))

(fn Update
  (params (delta :type Float%20%28double-precision%29))
  (bind _returnvalue (CallFunction|GetIdleVFXComponent))
  (Utilities|FlowControl|Sequence
    (:then_0
      (bind _output_get (LocalVariables|SetAssocSkelMesh (Customization|GetSkeletalMeshforPartType (CustomCharacter|GetAssociatedPlayerPawn) (CustomCharacter|GetCharacterPartType) :!class /Script/FortniteGame.FortPlayerPawn)))
      (bind _ascustom_character_part_anim_instance (Utilities|Casting|CastToCustomCharacterPartAnimInstance (Components|SkeletalMesh|GetAnimInstance _output_get)))
      (Utilities|IsValid _returnvalue
        (:Is%20Valid
          (bind _output_get_2 (LocalVariables|SetLOD0Check (> (Variables|/Script/FortniteGame.CustomCharacterPartAnimInstance|GetCurrentLOD _ascustom_character_part_anim_instance) 0)))
          (Effects|Components|ParticleSystem|SetBoolParameter _returnvalue "LOD0Check" _output_get_2)
          (Effects|Components|ParticleSystem|SetBoolParameter (CallFunction|GetIdleVFXComponent) "IsUsingRBD" (Variables|/Script/FortniteGame.FortAnimInstance|GetbRigidBodyEnabled _ascustom_character_part_anim_instance)))))
    (:then_1
      (bind _output_get_3 (Variables|Default|SetRunningWeight (Math|Interpolation|FInterpTo (Variables|Default|GetRunningWeight) (Utilities|Select :Index (Variables|/Script/Engine.CharacterMovementComponent|GetMovementMode (Variables|/Script/Engine.Character|GetCharacterMovement (CustomCharacter|GetAssociatedPlayerPawn))) :MOVE_None 0.0 :MOVE_Walking (Math|Float|MapRangeClamped (Math|Vector|VectorLengthXY (Transformation|GetVelocity (CustomCharacter|GetAssociatedPlayerPawn))) 1e+02 2.5e+02 0.0 1.0) :MOVE_NavWalking (Math|Float|MapRangeClamped (Math|Vector|VectorLengthXY (Transformation|GetVelocity (CustomCharacter|GetAssociatedPlayerPawn))) 1e+02 2.5e+02 0.0 1.0) :MOVE_Falling 0.0 :MOVE_Swimming 0.0 :MOVE_Flying 0.0 :MOVE_Custom 0.0) (Variables|Getdelta) 5.0)))
      (Rendering|Material|SetScalarParameterValueonMaterials (LocalVariables|GetAssocSkelMesh) "AstralHair_IsRunning" (Math|Float|SelectFloat _output_get_3 0.0 (or (LocalVariables|GetLOD0Check) (Variables|Default|GetHasBackbling)))))))

(fn ConstructionScript
  (|CallParentFunction))
