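; Character part modifier for the female Black Knight cosmetic that drives season-level "super levelling" visuals on the part's mesh.
; It watches the SeasonLevel13 stat, keeps the highest value, and pushes level range and current level into BKSL material scalars.
; Between start and end levels it stamps game time and, for Body parts, spawns a variant-applied Niagara effect behind a DoOnce.
; A self-labelled hacky 0.1-second timer re-applies parameters and is cleared after a one-second delay post-init.
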(event Custom|HackyWorkaroundRefreshTimerGuy
  ;[1 1 1] SuperLevelling Logic
  (Utilities|FlowControl|Sequence
    (:then_0
      (CallFunction|InitializeVariabels)
      (Rendering|Material|SetScalarParameterValueonMaterials (CustomCharacter|GetSkeletalMeshForPartOnPlayerPawn) "BKSL_CurrentLevel" (Math|Conversions|ToFloat(Integer) (stale EventOnPostResIn _output_get)))
      (Utilities|FlowControl|DoOnce false
        (:Completed
          (bind _returnvalue (Utilities|Time|SetTimerbyEvent (delegate Custom|HackyWorkaroundRefreshTimerGuy) 0.1 true))
          (Variables|Default|SetSL_RefreshTimeHandler _returnvalue))))
    (:then_1
      (Utilities|FlowControl|Gate true
        (:Exit
          (if (and (> (Math|Integer|Max(Integer) (stale EventOnPostResIn _statvalue) (Variables|Default|GetSL_LastValue)) (Variables|Default|GetSL_StartLevel)) (< (Math|Integer|Max(Integer) (stale EventOnPostResIn _statvalue) (Variables|Default|GetSL_LastValue)) (Variables|Default|GetSL_EndLevel)))
            (CallFunction|OnActualStatChange)))))))

(event EventOnPostResIn
  (bind _returnvalue (CustomCharacter|GetSkeletalMeshForPartOnPlayerPawn))
  (bind _sl_lastvalue (Variables|Default|GetSL_LastValue))
  (|CallParentFunction)
  (CallFunction|ResetSafetyGate)
  ;[1 1 1] SuperLevelling Logic
  (Utilities|FlowControl|Sequence
    (:then_0
      (Utilities|FlowControl|Delay 1.0)
      (Utilities|Time|ClearandInvalidateTimerbyHandle (Variables|Default|GetSL_RefreshTimeHandler)))
    (:then_1
      (bind (_statname _statvalue) (Cosmetics|WatchCosmeticStat (CustomCharacter|GetAssociatedPlayerPawn) "(TagName=\"StatManager.Season.SeasonLevel13\")")
        (:OnStatChanged
          (Rendering|Material|SetScalarParameterValueonMaterials _returnvalue "BKSL_StartAtLevel" (Math|Conversions|ToFloat(Integer) (Variables|Default|GetSL_StartLevel)))
          (Rendering|Material|SetScalarParameterValueonMaterials _returnvalue "BKSL_StopAtLevel" (Math|Conversions|ToFloat(Integer) (Variables|Default|GetSL_EndLevel)))
          (Rendering|Material|SetScalarParameterValueonMaterials _returnvalue "BKSL_EndRadius" (Variables|Default|GetEndRadius))
          (bind _output_get (Variables|Default|SetSL_LastValue (Math|Integer|Max(Integer) _statvalue _sl_lastvalue)))
          (Utilities|FlowControl|Sequence
            (:then_0
              (CallFunction|InitializeVariabels)
              (Rendering|Material|SetScalarParameterValueonMaterials (CustomCharacter|GetSkeletalMeshForPartOnPlayerPawn) "BKSL_CurrentLevel" (Math|Conversions|ToFloat(Integer) _output_get))
              (Utilities|FlowControl|DoOnce false
                (:Completed
                  (bind _returnvalue_2 (Utilities|Time|SetTimerbyEvent (delegate Custom|HackyWorkaroundRefreshTimerGuy) 0.1 true))
                  (Variables|Default|SetSL_RefreshTimeHandler _returnvalue_2))))
            (:then_1
              (Utilities|FlowControl|Gate true
                (:Exit
                  (if (and (> (Math|Integer|Max(Integer) _statvalue _sl_lastvalue) (Variables|Default|GetSL_StartLevel)) (< (Math|Integer|Max(Integer) _statvalue _sl_lastvalue) (Variables|Default|GetSL_EndLevel)))
                    (CallFunction|OnActualStatChange)))))))))))

(event Custom|ResetSafetyGate
  ;[1 1 1] SuperLevelling Logic
  (Utilities|FlowControl|DoOnce false
    (:Completed
      (bind _returnvalue (Utilities|Time|SetTimerbyEvent (delegate Custom|HackyWorkaroundRefreshTimerGuy) 0.1 true))
      (Variables|Default|SetSL_RefreshTimeHandler _returnvalue))))

(event Custom|OnActualStatChange
  ;[1 1 1] SuperLevelling Logic
  ;;[1 1 1] This should only fire off when the stat changes - use this event!
  (Rendering|Material|SetScalarParameterValueonMaterials (CustomCharacter|GetSkeletalMeshForPartOnPlayerPawn) "BKSL_GameTimeFromBP" (Utilities|Time|GetGameTimeinSeconds))
  ;[1 0 0] Spawn FX, this fires on character spawn. I would like to stop that.
  (if (and (Utilities|IsValid (CustomCharacter|GetSkeletalMeshForPartOnPlayerPawn) :!class /Script/Engine.KismetSystemLibrary) (Utilities|Enum|Equal(Enum) (CustomCharacter|GetCharacterPartType) "Body"))
    ;;[1 1 1] This is a saftey
    (Utilities|FlowControl|DoOnce false
      (:Completed
        (bind _returnvalue (PartModifier|SpawnSystem|SpawnSystemAttachedPartModifier "/BRCosmetics/Effects/Fort_Effects/Effects/Characters/Athena_Parts/BlackKnight/NS_BlackKnight_SuperLeveling.NS_BlackKnight_SuperLeveling" (CustomCharacter|GetSkeletalMeshForPartOnPlayerPawn) "root" "0, 0, 0" "0, 0, 0" "KeepRelativeOffset" true false)
          (:Success
            (Cosmetics|ApplyVariantsToComponent :self (CosmeticLoadout|GetItemDefinitionfromLoadout (CosmeticLoadout|GetCosmeticLoadout) "(TagName=\"LoadoutSlot.Character\")") :InComponent _returnvalue :Loadout (CosmeticLoadout|GetCosmeticLoadout))
            (Niagara|SetNiagaraVariable(Float) _returnvalue "User.BKSL_StartAtLevel" (Math|Conversions|ToFloat(Integer) (Variables|Default|GetSL_StartLevel)))
            (Niagara|SetNiagaraVariable(Float) _returnvalue "User.BKSL_StopAtLevel" (Math|Conversions|ToFloat(Integer) (Variables|Default|GetSL_EndLevel)))
            (Niagara|SetNiagaraVariable(Float) _returnvalue "User.BKSL_CurrentLevel" (Math|Conversions|ToFloat(Integer) (Variables|Default|GetSL_LastValue)))
            (Components|Activation|Activate _returnvalue true))
          (:Failure
            _))
        (Utilities|FlowControl|Delay 1.0)
        (CallFunction|OpenNiagaraGate)))))

(event Custom|Open%20Niagara%20Gate
  ;[1 0 0] Spawn FX, this fires on character spawn. I would like to stop that.
  ;;[1 1 1] This is a saftey
  (Utilities|FlowControl|DoOnce false
    (:Completed
      (bind _returnvalue (PartModifier|SpawnSystem|SpawnSystemAttachedPartModifier "/BRCosmetics/Effects/Fort_Effects/Effects/Characters/Athena_Parts/BlackKnight/NS_BlackKnight_SuperLeveling.NS_BlackKnight_SuperLeveling" (CustomCharacter|GetSkeletalMeshForPartOnPlayerPawn) "root" "0, 0, 0" "0, 0, 0" "KeepRelativeOffset" true false)
        (:Success
          (Cosmetics|ApplyVariantsToComponent :self (CosmeticLoadout|GetItemDefinitionfromLoadout (CosmeticLoadout|GetCosmeticLoadout) "(TagName=\"LoadoutSlot.Character\")") :InComponent _returnvalue :Loadout (CosmeticLoadout|GetCosmeticLoadout))
          (Niagara|SetNiagaraVariable(Float) _returnvalue "User.BKSL_StartAtLevel" (Math|Conversions|ToFloat(Integer) (Variables|Default|GetSL_StartLevel)))
          (Niagara|SetNiagaraVariable(Float) _returnvalue "User.BKSL_StopAtLevel" (Math|Conversions|ToFloat(Integer) (Variables|Default|GetSL_EndLevel)))
          (Niagara|SetNiagaraVariable(Float) _returnvalue "User.BKSL_CurrentLevel" (Math|Conversions|ToFloat(Integer) (Variables|Default|GetSL_LastValue)))
          (Components|Activation|Activate _returnvalue true))
        (:Failure
          _))
      (Utilities|FlowControl|Delay 1.0)
      (CallFunction|OpenNiagaraGate))))

(event Custom|InitializeVariabels
  (bind _returnvalue (CustomCharacter|GetSkeletalMeshForPartOnPlayerPawn))
  (Rendering|Material|SetScalarParameterValueonMaterials _returnvalue "BKSL_StartAtLevel" (Math|Conversions|ToFloat(Integer) (Variables|Default|GetSL_StartLevel)))
  (Rendering|Material|SetScalarParameterValueonMaterials _returnvalue "BKSL_StopAtLevel" (Math|Conversions|ToFloat(Integer) (Variables|Default|GetSL_EndLevel)))
  (Rendering|Material|SetScalarParameterValueonMaterials _returnvalue "BKSL_EndRadius" (Variables|Default|GetEndRadius)))

(fn ApplyLevelParam
  (params (Level :type Integer))
  (Rendering|Material|SetScalarParameterValueonMaterials (CustomCharacter|GetSkeletalMeshForPartOnPlayerPawn) "SL_CurrentLevel" (Math|Conversions|ToFloat(Integer) Level)))

(fn ConstructionScript
  (bind _returnvalue (CustomCharacter|GetSkeletalMeshForPartOnPlayerPawn))
  (|CallParentFunction)
  (Rendering|Material|SetScalarParameterValueonMaterials (CustomCharacter|GetSkeletalMeshForPartOnPlayerPawn) "SL_GameTimeFromBP" -99999.0)
  (Rendering|Material|SetScalarParameterValueonMaterials _returnvalue "SL_StartAtLevel" (Math|Conversions|ToFloat(Integer) (Variables|Default|GetSL_StartLevel)))
  (Rendering|Material|SetScalarParameterValueonMaterials _returnvalue "SL_StopAtLevel" (Math|Conversions|ToFloat(Integer) (Variables|Default|GetSL_EndLevel))))
