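; Part modifier for the EmberRae back bling that makes it react to the player's eliminations and chest openings.
; On post-init, it enables reactivity when the Parts variant check for Stage1 returns valid, and watches the kill and chest cosmetic stats.
; Kills play an attached sound only if reactivity is enabled.
; Chest opens, ungated, play the same sound after a short delay, with a timeline driving emissive material scalars.
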(event Custom|OnKill
  ;[1 1 1] OnKill Event
  (if (Variables|Default|GetReactivityEnabled)
    (CallFunction|SetandPlaySound)))

(event EventOnPostResIn
  (bind _returnvalue (CustomCharacter|GetSkeletalMeshForPartOnPlayerPawn))
  (|CallParentFunction)
  (Utilities|FlowControl|Sequence
    (:then_0
      ;[1 1 1] Elimination
      (Cosmetics|WatchCosmeticStat (CustomCharacter|GetAssociatedPlayerPawn) "(TagName=\"StatManager.AthenaKills\")" 0 false
        (:OnStatChanged
          (CallFunction|OnKill))))
    (:then_1
      ;[1 1 1] Read CIDs
      (bind (_index _bindexisvalid) (Class|FortCosmeticFunctionLibrary|GetVariantChannelActiveIndexForItem "/BRCosmetics/Athena/Items/Cosmetics/Backpacks/Backpack_EmberRae.Backpack_EmberRae" (CosmeticLoadout|GetCosmeticLoadout) "(TagName=\"Cosmetics.Variant.Channel.Parts\")" (Utilities|Array|MakeArray "(TagName=\"Cosmetics.Variant.Property.Stage1\")") false false))
      (Variables|Default|SetReactivityEnabled _bindexisvalid))
    (:then_2
      ;[1 1 1] openChest
      (Cosmetics|WatchCosmeticStat (CustomCharacter|GetAssociatedPlayerPawn) "(TagName=\"StatManager.OpenChests\")" 0 false
        (:OnStatChanged
          (Utilities|FlowControl|Delay 0.2)
          (CallFunction|SetandPlaySound)
          (bind (_direction _reactivitytime _belly_radius) (Timeline|Emissive PlayFromStart)
            (:Update
              (Rendering|Material|SetScalarParameterValueonMaterials _returnvalue "MaterialCurveA" _reactivitytime)
              (Rendering|Material|SetScalarParameterValueonMaterials _returnvalue "Belly Radius" _belly_radius))))))))

(fn OnKill)

(fn SetAndPlaySound
  (Audio|SpawnSoundAttached "/BRCosmetics/Sounds/Backblings/EmberRae/BackBling_EmberRae_Elimination_Cue.BackBling_EmberRae_Elimination_Cue" (CustomCharacter|GetSkeletalMeshForPartOnPlayerPawn)))

(fn ConstructionScript)
