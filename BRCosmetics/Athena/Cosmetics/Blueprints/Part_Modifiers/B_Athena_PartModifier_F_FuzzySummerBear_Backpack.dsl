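; Part modifier for the Fuzzy Summer Bear backpack that makes its bear face and heart material effects react to player events.
; After res-in it resets the effect parameters, then watches the DamageReceived and OpenChests cosmetic stats.
; Each stat change sets an emote visibility value and plays a timeline driving heart color, FX mask and gametime scalars on the part mesh.
; In the frontend, a looping timer previews the open-chest, damage and default states in turn.
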(event Custom|frontend_damage
  (bind _returnvalue (CustomCharacter|GetSkeletalMeshForPartOnPlayerPawn))
  (Rendering|Material|SetScalarParameterValueonMaterials (CustomCharacter|GetSkeletalMeshForPartOnPlayerPawn) "Bear summer Emote visibily" 1e+01)
  (Rendering|Material|SetScalarParameterValueonMaterials (CustomCharacter|GetSkeletalMeshForPartOnPlayerPawn) "BearSummer_gametime" 1.0)
  (Rendering|Material|SetScalarParameterValueonMaterials _returnvalue "Neon_GameTimeFromBP" (Utilities|Time|GetGameTimeinSeconds))
  (Rendering|Material|SetScalarParameterValueonMaterials _returnvalue "Neon_ReactiveLength" 3.0)
  (bind (_direction _receive_damage _heartfxmask) (Timeline|Timeline_2 PlayFromStart)
    (:Update
      (Rendering|Material|SetScalarParameterValueonMaterials _returnvalue "HeartShapeColor" _receive_damage)
      (Rendering|Material|SetScalarParameterValueonMaterials (CustomCharacter|GetSkeletalMeshForPartOnPlayerPawn) "HeartFxMask" _heartfxmask))))

(event Custom|FrontendOpenChest
  (bind _returnvalue (CustomCharacter|GetSkeletalMeshForPartOnPlayerPawn))
  (Rendering|Material|SetScalarParameterValueonMaterials _returnvalue "Bear summer Emote visibily" 5.0)
  (Rendering|Material|SetScalarParameterValueonMaterials (CustomCharacter|GetSkeletalMeshForPartOnPlayerPawn) "BearSummer_gametime" 1.0)
  (bind (_direction _openchest_heart _openchest_bear _heartfxmask) (Timeline|Timeline_3 PlayFromStart)
    (:Update
      (Rendering|Material|SetScalarParameterValueonMaterials _returnvalue "HeartShapeColor" _openchest_heart)
      (Rendering|Material|SetScalarParameterValueonMaterials (CustomCharacter|GetSkeletalMeshForPartOnPlayerPawn) "HeartFxMask" _heartfxmask))))

(event Custom|frontend%20Default
  (bind _returnvalue (CustomCharacter|GetSkeletalMeshForPartOnPlayerPawn))
  (bind _returnvalue_2 (CustomCharacter|GetSkeletalMeshForPartOnPlayerPawn))
  (Rendering|Material|SetScalarParameterValueonMaterials _returnvalue "Bear summer Emote visibily" 0.0)
  (Rendering|Material|SetScalarParameterValueonMaterials _returnvalue "HeartShapeColor" 0.0)
  (bind (_direction _openchest_heart _openchest_bear _heartfxmask) (Timeline|Timeline PlayFromStart)
    (:Update
      (Rendering|Material|SetScalarParameterValueonMaterials _returnvalue_2 "HeartFxMask" _heartfxmask)
      (Rendering|Material|SetScalarParameterValueonMaterials _returnvalue_2 "BearSummer_gametime" 1.0))))

(event EventOnPostResIn
  (bind _returnvalue (CustomCharacter|GetSkeletalMeshForPartOnPlayerPawn))
  (bind _returnvalue_2 (CustomCharacter|GetSkeletalMeshForPartOnPlayerPawn))
  (bind _returnvalue_3 (CustomCharacter|GetSkeletalMeshForPartOnPlayerPawn))
  (bind _returnvalue_4 (CustomCharacter|GetSkeletalMeshForPartOnPlayerPawn))
  (bind _returnvalue_5 (CustomCharacter|GetSkeletalMeshForPartOnPlayerPawn))
  (bind _returnvalue_6 (CustomCharacter|GetSkeletalMeshForPartOnPlayerPawn))
  (|CallParentFunction)
  (Utilities|FlowControl|Sequence
    (:then_0
      ;[1 1 1] Damage Received
      (Cosmetics|WatchCosmeticStat (CustomCharacter|GetAssociatedPlayerPawn) "(TagName=\"StatManager.DamageReceived\")" 0 false
        (:OnStatChanged
          (Rendering|Material|SetScalarParameterValueonMaterials (CustomCharacter|GetSkeletalMeshForPartOnPlayerPawn) "Bear summer Emote visibily" 1e+01)
          (Rendering|Material|SetScalarParameterValueonMaterials _returnvalue_4 "Neon_GameTimeFromBP" (Utilities|Time|GetGameTimeinSeconds))
          (Rendering|Material|SetScalarParameterValueonMaterials _returnvalue_4 "Neon_ReactiveLength" 3.0)
          (bind (_direction _receive_damage _heartfxmask) (Timeline|Timeline_1 PlayFromStart)
            (:Update
              (Rendering|Material|SetScalarParameterValueonMaterials _returnvalue_4 "HeartShapeColor" _receive_damage)
              (Rendering|Material|SetScalarParameterValueonMaterials _returnvalue "BearSummer_gametime" _receive_damage)
              (Rendering|Material|SetScalarParameterValueonMaterials _returnvalue "HeartFxMask" _heartfxmask))))))
    (:then_1
      ;[1 1 1] openChest
      (Cosmetics|WatchCosmeticStat (CustomCharacter|GetAssociatedPlayerPawn) "(TagName=\"StatManager.OpenChests\")" 0 false
        (:OnStatChanged
          (Rendering|Material|SetScalarParameterValueonMaterials _returnvalue_2 "Bear summer Emote visibily" 5.0)
          (bind (_direction_2 _openchest_heart _openchest_bear _heartfxmask_2) (Timeline|Timeline_0 PlayFromStart)
            (:Update
              (Rendering|Material|SetScalarParameterValueonMaterials _returnvalue_2 "HeartShapeColor" _openchest_heart)
              (Rendering|Material|SetScalarParameterValueonMaterials _returnvalue_3 "BearSummer_gametime" _openchest_bear)
              (Rendering|Material|SetScalarParameterValueonMaterials _returnvalue_3 "HeartFxMask" _heartfxmask_2))))))
    (:then_2
      ;[1 1 1] Stop Reaction on Spawn + Reduce Audio Amplitude in FE
      (Rendering|Material|SetScalarParameterValueonMaterials _returnvalue_5 "Bear summer Emote visibily" 0.0)
      (Rendering|Material|SetScalarParameterValueonMaterials _returnvalue_5 "HeartShapeColor" 0.0)
      (Rendering|Material|SetScalarParameterValueonMaterials _returnvalue_6 "HeartFxMask" 0.0)
      (Rendering|Material|SetScalarParameterValueonMaterials _returnvalue_6 "BearSummer_gametime" 1.0))
    (:then_3
      (if (CustomCosmeticModifier|IsinFrontEnd)
        (Utilities|Time|SetTimerbyEvent (delegate Custom|frontend%20circle) 6.2 true false -4.0)))))

(event Custom|frontend%20circle
  (CallFunction|FrontendOpenChest)
  (Utilities|FlowControl|Delay 2.0)
  (CallFunction|FrontendDamage)
  (Utilities|FlowControl|Delay 2.0)
  (CallFunction|FrontendDefault)
  (Utilities|FlowControl|Delay 2.0))

(fn ConstructionScript)
