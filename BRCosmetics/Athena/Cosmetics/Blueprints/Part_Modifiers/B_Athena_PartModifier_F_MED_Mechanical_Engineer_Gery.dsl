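; Cosmetic character part modifier for the F_MED_Mechanical_Engineer_Gery part that makes its effects react to the player aiming.
; When targeting or first-person camera state changes, it reads the pawn's bIsTargeting flag.
; Targeting activates the idle VFX and raises the part mesh's "emissive mult" material scalar; otherwise the VFX deactivates and the glow dims.
; After res-in, it calls the parent handler, then deactivates the idle VFX.
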(event EventOnPostResIn
  (|CallParentFunction)
  (Components|Activation|Deactivate (CallFunction|GetIdleVFXComponent)))

(event EventOnTargetingOrFirstPersonCameraChanged
  (bind _returnvalue (CallFunction|GetIdleVFXComponent))
  (bind _returnvalue_2 (CustomCharacter|GetSkeletalMeshForPartOnPlayerPawn))
  (if (Variables|/Script/FortniteGame.FortPlayerPawn|GetbIsTargeting (CustomCharacter|GetAssociatedPlayerPawn))
    (Components|Activation|Activate _returnvalue false)
    (Rendering|Material|SetScalarParameterValueonMaterials _returnvalue_2 "emissive mult" 2.0)
    (else
      (Components|Activation|Deactivate _returnvalue)
      (Rendering|Material|SetScalarParameterValueonMaterials _returnvalue_2 "emissive mult" 0.2))))

(fn ConstructionScript)
