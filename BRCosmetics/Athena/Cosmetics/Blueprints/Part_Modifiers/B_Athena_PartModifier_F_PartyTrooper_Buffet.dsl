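; Cosmetic part modifier for the female Party Trooper (Buffet) character part that drives a damage reaction in its idle particle effect.
; When a stat change reports a positive value, it stamps the current game time into the effect's DamageTime parameter.
; In the front end, after a short delay, it caches the idle VFX component and starts a looping timer that fakes damage by setting duration and time parameters.
; Destruction clears that timer.
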(event EventOnStatChanged
  (params (State%20Name :type Name :default "None") (Stat%20Value :type Integer))
  (if (> Stat%20Value 0)
    (Effects|Components|ParticleSystem|SetFloatParameter (CallFunction|GetIdleVFXComponent) "DamageTime" (Utilities|Time|GetGameTimeinSeconds))))

(event EventDestroyed
  (Utilities|Time|ClearandInvalidateTimerbyHandle (stale EventOnPostResIn _returnvalue)))

(event EventOnPostResIn
  (|CallParentFunction)
  (if (CustomCosmeticModifier|IsinFrontEnd)
    (Utilities|FlowControl|Delay 0.3)
    (Variables|Default|SetIdleVFXComponent (CallFunction|GetIdleVFXComponent)) ; from blue port - "promote to variable"
    (bind _returnvalue (Utilities|Time|SetTimerbyEvent (delegate Custom|FakeDamage) 6.0 true))))

(event Custom|FakeDamage
  (bind _idlevfxcomponent (Variables|Default|GetIdleVFXComponent))
  (Effects|Components|ParticleSystem|SetFloatParameter _idlevfxcomponent "DamageDuration" 3.0)
  (Effects|Components|ParticleSystem|SetFloatParameter _idlevfxcomponent "DamageTime" (Utilities|Time|GetGameTimeinSeconds)))

(fn ConstructionScript)
