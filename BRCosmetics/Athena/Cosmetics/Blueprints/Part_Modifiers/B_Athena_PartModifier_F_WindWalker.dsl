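; Cosmetic part modifier for the WindWalker character part that animates an emissive glow on the pawn's skeletal mesh for that part type.
; After post-res-in, the front end loops a timeline that drives the "emissive mult" material scalar, scaled by GlowStrength.
; In game, it watches the OpenChests cosmetic stat and plays the GlowCurve timeline once on change.
; BeginPlay also triggers GlowCurve through a do-once gate.
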(event Collision|EventActorBeginOverlap
  (params (OtherActor :type Actor%20Object%20Reference)))

(event EventTick
  (params (DeltaSeconds :type Float%20%28single-precision%29)))

(event EventOnPostResIn
  (bind _returnvalue (CustomCharacter|GetAssociatedPlayerPawn))
  (bind _returnvalue_2 (CustomCharacter|GetCharacterPartType))
  (bind _returnvalue_3 (Customization|GetSkeletalMeshforPartType _returnvalue _returnvalue_2 :!class /Script/FortniteGame.FortPlayerPawn))
  (|CallParentFunction)
  (if (CustomCosmeticModifier|IsinFrontEnd)
    (bind (_direction _frontendglowcurve) (Timeline|FrontEndTimeline PlayFromStart)
      (:Update
        (Rendering|Material|SetScalarParameterValueonMaterials _returnvalue_3 "emissive mult" (* _frontendglowcurve (Variables|Default|GetGlowStrength))))
      (:Finished
        (CallFunction|ReplayLoop)))
    (else
      ;[1 1 1] Set Mesh Materials
      (Cosmetics|WatchCosmeticStat _returnvalue "(TagName=\"StatManager.OpenChests\")"
        (:OnStatChanged
          (Utilities|DoOnceThenDo false
            (:ThenDo
              (bind (_direction_2 _emissivetrack) (Timeline|GlowCurve PlayFromStart)
                (:Update
                  (Rendering|Material|SetScalarParameterValueonMaterials _returnvalue_3 "emissive mult" (* _emissivetrack (Variables|Default|GetGlowStrength))))))))))))

(event EventBeginPlay
  ;[1 1 1] Set Mesh Materials
  (Utilities|DoOnceThenDo false
    (:ThenDo
      (Timeline|GlowCurve PlayFromStart))))

(event Custom|ReplayLoop
  (Timeline|FrontEndTimeline PlayFromStart))

(fn GameTimeFromBPForGlow
  (params (A :type Float%20%28double-precision%29) (B :type Float%20%28double-precision%29))
  (return 0.0))

(fn ConstructionScript)
