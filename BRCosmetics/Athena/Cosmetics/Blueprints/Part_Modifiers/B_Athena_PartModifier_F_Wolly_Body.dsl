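; Body part modifier for the Wolly female cosmetic that plays a character effect when the pawn lands.
; On the landed event it runs the parent logic, then spawns the Wolly particle system attached to the head of the part's skeletal mesh on the player pawn.
; The emitter auto-destroys, matches the pawn's custom depth and is debounced.
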(event EventOnPawnLanded
  (|CallParentFunction)
  (PartModifier|SpawnSystem|SpawnEmitterAttachedPartModifier(Legacy) :EmitterTemplate "/BRCosmetics/Characters/Player/Female/Medium/Bodies/F_MED_Wolly/FX/P_F_MED_Wolly.P_F_MED_Wolly" :AttachToComponent (CustomCharacter|GetSkeletalMeshForPartOnPlayerPawn) :AttachPointName "head" :Scale "1.000000,1.000000,1.000000" :LocationType "SnapToTargetIncludingScale" :bAutoDestroy true :PoolingMethod "None" :bAutoActivate true :bMatchPawnCustomDepth true :bDebounce true :DebounceTime -1.0))

(fn ConstructionScript)
