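; Cosmetic part modifier for Famine that drives the eye glow and smoke trail parameters on the character's idle VFX particle component.
; On day phase changes, a switch sets eye glow and light strength, with a brighter set in the Day branch and a dimmer set outside it.
; Skydiving begin and end events call their parent, then turn the black smoke trail spawn rate off and back on.
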(event EventOnDayPhaseChanged
  (params (CurrentDayPhase :type EFortDayPhase%20Enum :default "Morning") (PreviousDayPhase :type EFortDayPhase%20Enum :default "Morning") (bAtCreation :type Boolean))
  (bind _returnvalue (CallFunction|GetIdleVFXComponent))
  ;[1 1 1] Controls Emissive Values using Time of Day
  (switch Utilities|FlowControl|Switch|SwitchonEFortDayPhase CurrentDayPhase
    (:Morning
      _)
    (:Day
      (Effects|Components|ParticleSystem|SetFloatParameter _returnvalue "EyeGlowStrength" 0.35)
      (Effects|Components|ParticleSystem|SetFloatParameter _returnvalue "LightStrength" 3.0))
    (:Evening
      _)
    (:Night
      _))
  (Effects|Components|ParticleSystem|SetFloatParameter _returnvalue "EyeGlowStrength" 0.125)
  (Effects|Components|ParticleSystem|SetFloatParameter _returnvalue "LightStrength" 2.0))

(event EventOnBeginSkydiving
  ;[1 1 1] On and Off Switch when SkyDiving
  (|CallParentFunction)
  (Effects|Components|ParticleSystem|SetFloatParameter (CallFunction|GetIdleVFXComponent) "BlackSmokeTrailSpawnRate" 0.0))

(event EventOnEndSkydiving
  ;[1 1 1] On and Off Switch when SkyDiving
  (|CallParentFunction)
  (Effects|Components|ParticleSystem|SetFloatParameter (CallFunction|GetIdleVFXComponent) "BlackSmokeTrailSpawnRate" 1e+01))

(fn ConstructionScript)
