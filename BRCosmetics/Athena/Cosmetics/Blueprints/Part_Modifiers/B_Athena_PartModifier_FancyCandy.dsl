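; Cosmetic character part modifier for FancyCandy that makes the part's idle particle effect react to the player pawn's movement.
; On each movement update it writes the pawn's speed to the RunningSpeed particle parameter, snapping it to zero when velocity is nearly zero.
; On landing it plays a timeline from the start and feeds the curve value into the Land parameter of the same effect.
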(event EventOnCharacterMovementUpdated
  (params (DeltaSeconds :type Float%20%28single-precision%29) (OldLocation :type Vector :default "0, 0, 0") (OldVelocity :type Vector :default "0, 0, 0"))
  (bind _returnvalue (CustomCharacter|GetAssociatedPlayerPawn))
  (bind _returnvalue_2 (Transformation|GetVelocity _returnvalue))
  (bind _returnvalue_3 (CallFunction|GetIdleVFXComponent))
  (|CallParentFunction DeltaSeconds OldLocation OldVelocity)
  (if (Math|Vector|Equal(Vector) _returnvalue_2 "0, 0, 0" 0.0001)
    (Effects|Components|ParticleSystem|SetFloatParameter _returnvalue_3 "RunningSpeed" 0.0)
    (else
      (Effects|Components|ParticleSystem|SetFloatParameter _returnvalue_3 "RunningSpeed" (Math|Vector|VectorLength _returnvalue_2)))))

(event EventOnPawnLanded
  (|CallParentFunction)
  (bind (_direction _spawntime) (Timeline|Timeline_0 PlayFromStart)
    (:Update
      (Effects|Components|ParticleSystem|SetFloatParameter (CallFunction|GetIdleVFXComponent) "Land" _spawntime))))

(fn ConstructionScript
  (|CallParentFunction))
