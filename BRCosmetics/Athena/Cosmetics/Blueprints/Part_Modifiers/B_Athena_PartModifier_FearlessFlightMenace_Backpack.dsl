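; Cosmetic part modifier for the Fearless Flight Menace backpack that makes its materials react to the player's eliminations and down-but-not-out (DBNO) state.
; A watched kill stat stores the elimination count and plays a timeline driving WipeProgression.
; DBNO changes play or reverse a second timeline driving AlphaProgression and WipeProgression.
; In the front end, the kill animation plays on init and replays after a short delay as a looping preview.
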(event EventOnPostResIn
  (bind _returnvalue (CustomCharacter|GetAssociatedPlayerPawn))
  (|CallParentFunction)
  ;[0.2818 1 0.7667] Init Triggers
  (bind (_statname _statvalue) (Cosmetics|WatchCosmeticStat (CustomCharacter|GetAssociatedPlayerPawn) "(TagName=\"StatManager.AthenaKills\")" 0 false)
    (:then
      (Utilities|Casting|CastToFortPlayerPawnAthena _returnvalue)
      (EventDispatchers|BindEventtoOnDBNOStateChanged _returnvalue (delegate Custom|OnDBNO))
      (if (CustomCosmeticModifier|IsinFrontEnd)
        (CallFunction|PlayReactivityinFE)))
    (:OnStatChanged
      (Variables|Default|SetEliminations (Math|Conversions|ToFloat(Integer) _statvalue))
      (CallFunction|OnKill))))

(event Custom|OnKill
  ;[1 0.08854 0.08854] OnKill
  (bind (_direction _alpha) (Timeline|OnKillTL PlayFromStart)
    (:Update
      (Rendering|Material|SetScalarParameterValueonMaterials (CustomCharacter|GetSkeletalMeshForPartOnPlayerPawn) "WipeProgression" _alpha))
    (:Finished
      ;[0.116 1 0.3229] Play reactivity in FE
      (if (CustomCosmeticModifier|IsinFrontEnd)
        (Utilities|FlowControl|Delay 2.0)
        (CallFunction|PlayReactivityinFE)))))

(event Custom|OnDBNO
  (params (FortPawn :type Fort%20Pawn%20Object%20Reference) (bInIsDBNO :type Boolean))
  (bind _returnvalue (CustomCharacter|GetSkeletalMeshForPartOnPlayerPawn))
  ;[1 0.3424 0.7863] On DBNO
  (if bInIsDBNO
    (bind (_direction _alpha) (Timeline|DBNO_TL PlayFromStart)
      (:Update
        (Rendering|Material|SetScalarParameterValueonMaterials _returnvalue "AlphaProgression" _alpha)
        (Rendering|Material|SetScalarParameterValueonMaterials _returnvalue "WipeProgression" _alpha)))
    (else
      (Timeline|DBNO_TL ReverseFromEnd))))

(event Custom|PlayReactivityInFE
  ;[1 0.08854 0.08854] OnKill
  (Timeline|OnKillTL PlayFromStart))

(fn ConstructionScript
  (|CallParentFunction))
