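; Athena cosmetic part modifier Blueprint named FeatherTest, apparently a test modifier for a feather-related character part.
; Its post-res-in event and construction script only call the parent implementations, so the graphs shown add no logic of their own.
; Any visible behavior comes from the parent class or from settings that these graphs do not show.
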(event EventOnPostResIn
  (|CallParentFunction))

(fn ConstructionScript
  (|CallParentFunction))
