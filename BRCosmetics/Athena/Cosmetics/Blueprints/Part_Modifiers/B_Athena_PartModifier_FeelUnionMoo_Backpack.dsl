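; Cosmetic part modifier Blueprint for the FeelUnionMoo backpack that makes it react visually when a tracked player stat changes.
; On a stat change it calls the parent handler, raises the "React" scalar on the part's skeletal mesh materials, then resets it to zero after a fixed delay.
; The stat name and value are not checked here, and the KillReactivity function is empty.
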(event EventOnStatChanged
  (params (State%20Name :type Name :default "None") (Stat%20Value :type Integer))
  (bind _returnvalue (CustomCharacter|GetSkeletalMeshforPartonOwner))
  (|CallParentFunction State%20Name Stat%20Value)
  (Rendering|Material|SetScalarParameterValueonMaterials _returnvalue "React" 0.5)
  (Utilities|FlowControl|Delay 6.0)
  (Rendering|Material|SetScalarParameterValueonMaterials _returnvalue "React" 0.0))

(fn KillReactivity
  (params (KillCount :type Integer)))

(fn ConstructionScript
  (|CallParentFunction))
