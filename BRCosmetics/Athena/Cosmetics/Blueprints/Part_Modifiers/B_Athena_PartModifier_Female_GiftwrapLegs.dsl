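; Cosmetic part modifier Blueprint for the female Giftwrap legs that pulses a landing flag on its idle particle effect.
; When the pawn lands, it calls the parent, sets the idle VFX component's bhasLanded particle parameter true, then resets it to false after a short delay.
; Its construction script prints a leftover-looking debug string to screen and log before calling the parent.
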(event EventOnPawnLanded
  (bind _returnvalue (CallFunction|GetIdleVFXComponent))
  (|CallParentFunction)
  (Effects|Components|ParticleSystem|SetBoolParameter _returnvalue "bhasLanded" true)
  (Utilities|FlowControl|Delay 0.1)
  (Effects|Components|ParticleSystem|SetBoolParameter _returnvalue "bhasLanded" false))

(fn ConstructionScript
  (Development|PrintString "Running Constructor" true true "(R=1.000000,G=0.000000,B=0.614498,A=1.000000)" 1e+01)
  (|CallParentFunction))
