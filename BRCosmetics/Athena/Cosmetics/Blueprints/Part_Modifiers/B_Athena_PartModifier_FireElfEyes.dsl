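; Cosmetic character part modifier that adds fire elf eye effects, an eye trail and an eye glow, to the player pawn's head.
; After res-in, it attaches both components to the "head" socket of the part's skeletal mesh with relative transforms.
; On weapon changes and post-res-in part application, it calls the parent, then shows or hides only the glow based on bIsHeadHidden, leaving the trails untouched.
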(event EventOnPostResIn
  (bind _returnvalue (CustomCharacter|GetSkeletalMeshForPartOnPlayerPawn))
  (|CallParentFunction)
  (Transformation|AttachComponentToComponent (Variables|Default|GetEyeTrails) _returnvalue "head" "KeepRelative" "KeepRelative" "KeepRelative" true)
  (Transformation|AttachComponentToComponent (Variables|Default|GetS_FX_FireElf_EyeGlow) _returnvalue "head" "KeepRelative" "KeepRelative" "KeepRelative" true))

(event EventOnWeaponChanged
  (params (NewWeapon :type Fort%20Weapon%20Object%20Reference) (OldWeapon :type Fort%20Weapon%20Object%20Reference))
  (|CallParentFunction NewWeapon OldWeapon)
  (if (Variables|HiddenPartCheck|GetbIsHeadHidden)
    (Rendering|SetVisibility (Variables|Default|GetS_FX_FireElf_EyeGlow) false false)
    (else
      (Rendering|SetVisibility (Variables|Default|GetS_FX_FireElf_EyeGlow) true false))))

(event EventOnCosmeticPartAppliedPostResIn
  (params (Part :type Custom%20Character%20Part%20Object%20Reference))
  (|CallParentFunction Part)
  (if (Variables|HiddenPartCheck|GetbIsHeadHidden)
    (Rendering|SetVisibility (Variables|Default|GetS_FX_FireElf_EyeGlow) false false)
    (else
      (Rendering|SetVisibility (Variables|Default|GetS_FX_FireElf_EyeGlow) true false))))

(fn ConstructionScript)
