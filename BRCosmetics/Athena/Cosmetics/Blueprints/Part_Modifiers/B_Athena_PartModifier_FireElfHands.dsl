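; Cosmetic part modifier for the FireElf character that controls the material on its fire hands mesh.
; On weapon change and when the part is applied post res-in, it runs parent logic, then checks the hand-hidden flag.
; If hidden, it sets slot 0 to an invisible material.
; Otherwise it restores the fire hands material and reapplies the pawn's character loadout variants to that component.
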(event EventOnWeaponChanged
  (params (NewWeapon :type Fort%20Weapon%20Object%20Reference) (OldWeapon :type Fort%20Weapon%20Object%20Reference))
  (bind _returnvalue (CustomCharacter|GetSkeletalMeshForPartOnPlayerPawn))
  (|CallParentFunction NewWeapon OldWeapon)
  (if (Variables|HiddenPartCheck|GetbIsHandHidden)
    (Rendering|Material|SetMaterial _returnvalue 0 "/Game/Missions/Secondary/BluGlo/M_Invisible.M_Invisible")
    (else
      (Rendering|Material|SetMaterial _returnvalue 0 "/BRCosmetics/Characters/Player/Female/Medium/Bodies/F_MED_FireElf/Materials/M_FireHands.M_FireHands")
      (Class|FortCosmeticFunctionLibrary|ApplyVariantstoComponentbySlotTag (CustomCharacter|GetAssociatedPlayerPawn) "(TagName=\"LoadoutSlot.Character\")" _returnvalue))))

(event EventOnCosmeticPartAppliedPostResIn
  (params (Part :type Custom%20Character%20Part%20Object%20Reference))
  (bind _returnvalue (CustomCharacter|GetSkeletalMeshForPartOnPlayerPawn))
  (|CallParentFunction Part)
  (if (Variables|HiddenPartCheck|GetbIsHandHidden)
    (Rendering|Material|SetMaterial _returnvalue 0 "/Game/Missions/Secondary/BluGlo/M_Invisible.M_Invisible")
    (else
      (Rendering|Material|SetMaterial _returnvalue 0 "/BRCosmetics/Characters/Player/Female/Medium/Bodies/F_MED_FireElf/Materials/M_FireHands.M_FireHands")
      (Class|FortCosmeticFunctionLibrary|ApplyVariantstoComponentbySlotTag (CustomCharacter|GetAssociatedPlayerPawn) "(TagName=\"LoadoutSlot.Character\")" _returnvalue))))

(fn ConstructionScript)
