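; Cosmetic part modifier for the Dark Viking fire effect that changes character flame VFX when the pawn jumps and lands.
; On jump, it restarts a timeline whose curve, named as a 1-to-0 fade, drives the FlameSpawnRate parameter on the idle VFX component.
; On landing, it spawns an attached burst system using that component, stops the timeline and zeroes the spawn rate.
; Both handlers also call the parent implementation.
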(event EventOnPawnJumped
  (|CallParentFunction)
  (bind (_direction _curve1to0) (Timeline|Timeline_JumpCurve PlayFromStart)
    (:Update
      (Effects|Components|ParticleSystem|SetFloatParameter (CallFunction|GetIdleVFXComponent) "FlameSpawnRate" _curve1to0))))

(event EventOnPawnLanded
  (bind _returnvalue (CallFunction|GetIdleVFXComponent))
  (|CallParentFunction)
  (Utilities|FlowControl|Sequence
    (:then_0
      (PartModifier|SpawnSystem|SpawnSystemAttachedPartModifier (Variables|Default|GetNS_JumpLanding_Burst) _returnvalue))
    (:then_1
      (Timeline|Timeline_JumpCurve Stop))
    (:then_2
      (Effects|Components|ParticleSystem|SetFloatParameter _returnvalue "FlameSpawnRate" 0.0))))

(fn ConstructionScript
  (|CallParentFunction))
