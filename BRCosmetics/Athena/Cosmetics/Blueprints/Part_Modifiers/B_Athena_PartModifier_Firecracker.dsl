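; Character part modifier for the Firecracker cosmetic that adapts its idle VFX to the player's state.
; After res-in, it binds the idle Niagara effect to the skeletal mesh of the pawn's matching character part.
; Targeting or camera changes toggle the effect's MasterAlpha, while vehicle and skydiving events set driving and skydiving flags, the latter after a short delay.
; A custom QA event prints the effect's attach parent and socket.
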(event Custom|AttachmentTest
  ;[1 1 1] Added Print String Test for QA to see what's attached where
  (Development|PrintString (Utilities|String|Append "Idle VFX Attached to - " (Utilities|GetDisplayName (Transformation|GetAttachParent (CallFunction|GetIdleVFXComponent))) " - at Socket - " (Utilities|String|ToString(Name) (Transformation|GetAttachSocketName (CallFunction|GetIdleVFXComponent)))) true true "(R=1.000000,G=0.023320,B=0.000000,A=1.000000)" 1e+01))

(event EventOnPostResIn
  (|CallParentFunction)
  (Niagara|SetNiagaraSkeletalMeshComponent (Utilities|Casting|CastToNiagaraComponent (CallFunction|GetIdleVFXComponent)) "SkeletalMesh" (Customization|GetSkeletalMeshforPartType (CustomCharacter|GetAssociatedPlayerPawn) (CustomCharacter|GetCharacterPartType) :!class /Script/FortniteGame.FortPlayerPawn)))

(event EventOnTargetingOrFirstPersonCameraChanged
  (|CallParentFunction)
  (if (Variables|/Script/FortniteGame.FortPlayerPawn|GetbIsTargeting (CustomCharacter|GetAssociatedPlayerPawn))
    (Effects|Components|ParticleSystem|SetFloatParameter (CallFunction|GetIdleVFXComponent) "MasterAlpha" 0.0)
    (else
      (Effects|Components|ParticleSystem|SetFloatParameter (CallFunction|GetIdleVFXComponent) "MasterAlpha" 1.0))))

(event EventOnEnterVehicle
  (params (VehicleEntered :type Fort%20Vehicle%20Interface%20Interface%20%28by%20ref%29))
  (|CallParentFunction VehicleEntered)
  (Effects|Components|ParticleSystem|SetBoolParameter (CallFunction|GetIdleVFXComponent) "IsDriving" true))

(event EventOnExitVehicle
  (|CallParentFunction)
  (Effects|Components|ParticleSystem|SetBoolParameter (CallFunction|GetIdleVFXComponent) "IsDriving" false))

(event EventOnBeginSkydiving
  (|CallParentFunction)
  (Utilities|FlowControl|Delay 0.1)
  (Effects|Components|ParticleSystem|SetBoolParameter (CallFunction|GetIdleVFXComponent) "IsSkydiving" true))

(event EventOnEndSkydiving
  (|CallParentFunction)
  (Utilities|FlowControl|Delay 0.1)
  (Effects|Components|ParticleSystem|SetBoolParameter (CallFunction|GetIdleVFXComponent) "IsSkydiving" false))

(fn ConstructionScript
  (|CallParentFunction))
