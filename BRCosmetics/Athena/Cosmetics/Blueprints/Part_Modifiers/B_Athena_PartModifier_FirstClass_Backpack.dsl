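; Cosmetic part modifier for the FirstClass backpack that makes its face react to gameplay by toggling "Hide Element" material scalars on base and overlay materials.
; After spawn, it watches stats and emotes: kills, chests, healing, shields and emotes map to happy, damage to sad.
; Retriggerable delays restore the default face, while a looping timer, paused during reactions, flashes a bored face.
; In the frontend, it also adjusts toon hatching light parameters.
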(event Custom|TriggerHappyFace
  (params (MontageItemDef :type Fort%20Item%20Definition%20Object%20Reference) (PawnEmoting :type Fort%20Pawn%20Object%20Reference))
  (bind _returnvalue (CustomCharacter|GetSkeletalMeshForPartOnPlayerPawn))
  ;[1 1 1] Happy - Elimination and ChestOpen
  (Utilities|Time|PauseTimerbyHandle (Variables|Default|GetBoredTimer))
  (Rendering|Material|SetScalarParameterValueonMaterials _returnvalue "Hide Element 02" 1.0)
  (Class|FortCosmeticMaterialBlueprintLibrary|SetScalarParameterValueonOverlayMaterial _returnvalue "Hide Element 02" 1.0)
  (Rendering|Material|SetScalarParameterValueonMaterials _returnvalue "Hide Element 05" 0.0)
  (Class|FortCosmeticMaterialBlueprintLibrary|SetScalarParameterValueonOverlayMaterial _returnvalue "Hide Element 05")
  (Rendering|Material|SetScalarParameterValueonMaterials _returnvalue "Hide Element 03" 1.0) ; Disable Bored
  (Class|FortCosmeticMaterialBlueprintLibrary|SetScalarParameterValueonOverlayMaterial _returnvalue "Hide Element 03" 1.0) ; Disable Bored
  (Utilities|FlowControl|RetriggerableDelay 4.0)
  (Rendering|Material|SetScalarParameterValueonMaterials _returnvalue "Hide Element 02" 0.0)
  (Class|FortCosmeticMaterialBlueprintLibrary|SetScalarParameterValueonOverlayMaterial _returnvalue "Hide Element 02")
  (Rendering|Material|SetScalarParameterValueonMaterials _returnvalue "Hide Element 05" 1.0)
  (Class|FortCosmeticMaterialBlueprintLibrary|SetScalarParameterValueonOverlayMaterial _returnvalue "Hide Element 05" 1.0)
  (Utilities|Time|UnpauseTimerbyHandle (Variables|Default|GetBoredTimer)))

(event EventOnPostResIn
  (bind _returnvalue (CustomCharacter|GetAssociatedPlayerPawn))
  (bind _returnvalue_2 (CustomCharacter|GetSkeletalMeshForPartOnPlayerPawn))
  (bind _returnvalue_3 (CustomCharacter|GetSkeletalMeshForPartOnPlayerPawn))
  (bind _returnvalue_4 (CustomCharacter|GetSkeletalMeshForPartOnPlayerPawn))
  (|CallParentFunction)
  (Utilities|ToonFrontendLUTReplace
    (:Out
      (Utilities|FlowControl|Sequence
        (:then_0
          (if (CustomCosmeticModifier|IsinFrontEnd)
            ;[1 1 1] adjusting the direction vector for the hatching
            (Rendering|Material|SetScalarParameterValueonMaterials _returnvalue_2 "Threshold_Offset" -0.5)
            (Rendering|Material|SetScalarParameterValueonMaterials _returnvalue_2 "NdotL_Max" 1.0)
            (Rendering|Material|SetScalarParameterValueonMaterials _returnvalue_2 "NdotL_Min" 0.0)
            (Rendering|Material|SetScalarParameterValueonMaterials _returnvalue_2 "ThresholdVector_Strenght" 1.0)
            (Rendering|Material|SetVectorParameterValueonMaterials _returnvalue_2 "Threshold_Vector" "-1.000000,1.000000,1.000000")))
        (:then_1
          ;[1 1 1] Happy - Elimination and ChestOpen
          (Cosmetics|WatchCosmeticStat _returnvalue "(TagName=\"StatManager.AthenaKills\")" 0 false
            (:then
              (Cosmetics|WatchCosmeticStat _returnvalue "(TagName=\"StatManager.OpenChests\")" 0 false
                (:then
                  (Cosmetics|WatchCosmeticStat _returnvalue "(TagName=\"StatManager.GainedShieldTimes\")" 0 false
                    (:then
                      (Cosmetics|WatchCosmeticStat _returnvalue "(TagName=\"StatManager.GainedHealthTimes\")" 0 false
                        (:then
                          (EventDispatchers|BindEventtoOnPawnStartedEmote _returnvalue (delegate Custom|TriggerHappyFace)))
                        (:OnStatChanged
                          _)))
                    (:OnStatChanged
                      _)))
                (:OnStatChanged
                  _)))
            (:OnStatChanged
              _))
          (Utilities|Time|PauseTimerbyHandle (Variables|Default|GetBoredTimer))
          (Rendering|Material|SetScalarParameterValueonMaterials _returnvalue_3 "Hide Element 02" 1.0)
          (Class|FortCosmeticMaterialBlueprintLibrary|SetScalarParameterValueonOverlayMaterial _returnvalue_3 "Hide Element 02" 1.0)
          (Rendering|Material|SetScalarParameterValueonMaterials _returnvalue_3 "Hide Element 05" 0.0)
          (Class|FortCosmeticMaterialBlueprintLibrary|SetScalarParameterValueonOverlayMaterial _returnvalue_3 "Hide Element 05")
          (Rendering|Material|SetScalarParameterValueonMaterials _returnvalue_3 "Hide Element 03" 1.0) ; Disable Bored
          (Class|FortCosmeticMaterialBlueprintLibrary|SetScalarParameterValueonOverlayMaterial _returnvalue_3 "Hide Element 03" 1.0) ; Disable Bored
          (Utilities|FlowControl|RetriggerableDelay 4.0)
          (Rendering|Material|SetScalarParameterValueonMaterials _returnvalue_3 "Hide Element 02" 0.0)
          (Class|FortCosmeticMaterialBlueprintLibrary|SetScalarParameterValueonOverlayMaterial _returnvalue_3 "Hide Element 02")
          (Rendering|Material|SetScalarParameterValueonMaterials _returnvalue_3 "Hide Element 05" 1.0)
          (Class|FortCosmeticMaterialBlueprintLibrary|SetScalarParameterValueonOverlayMaterial _returnvalue_3 "Hide Element 05" 1.0)
          (Utilities|Time|UnpauseTimerbyHandle (Variables|Default|GetBoredTimer)))
        (:then_2
          ;[1 1 1] Sad - Damage Received
          (Cosmetics|WatchCosmeticStat (CustomCharacter|GetAssociatedPlayerPawn) "(TagName=\"StatManager.DamageReceived\")" 0 false
            (:OnStatChanged
              (Utilities|Time|PauseTimerbyHandle (Variables|Default|GetBoredTimer))
              (Rendering|Material|SetScalarParameterValueonMaterials _returnvalue_4 "Hide Element 02" 1.0)
              (Class|FortCosmeticMaterialBlueprintLibrary|SetScalarParameterValueonOverlayMaterial _returnvalue_4 "Hide Element 02" 1.0)
              (Rendering|Material|SetScalarParameterValueonMaterials _returnvalue_4 "Hide Element 04" 0.0)
              (Class|FortCosmeticMaterialBlueprintLibrary|SetScalarParameterValueonOverlayMaterial _returnvalue_4 "Hide Element 04")
              (Rendering|Material|SetScalarParameterValueonMaterials _returnvalue_4 "Hide Element 03" 1.0) ; Disable Bored
              (Class|FortCosmeticMaterialBlueprintLibrary|SetScalarParameterValueonOverlayMaterial _returnvalue_4 "Hide Element 03" 1.0) ; Disable Bored
              (Utilities|FlowControl|RetriggerableDelay 4.0)
              (Rendering|Material|SetScalarParameterValueonMaterials _returnvalue_4 "Hide Element 02" 0.0)
              (Class|FortCosmeticMaterialBlueprintLibrary|SetScalarParameterValueonOverlayMaterial _returnvalue_4 "Hide Element 02")
              (Rendering|Material|SetScalarParameterValueonMaterials _returnvalue_4 "Hide Element 04" 1.0)
              (Class|FortCosmeticMaterialBlueprintLibrary|SetScalarParameterValueonOverlayMaterial _returnvalue_4 "Hide Element 04" 1.0)
              (Utilities|Time|UnpauseTimerbyHandle (Variables|Default|GetBoredTimer)))))
        (:then_3
          ;[1 1 1] Bored
          (bind _returnvalue_5 (Utilities|Time|SetTimerbyEvent (delegate Custom|TriggerBoredFace) 6.666 true))
          (Variables|Default|SetBoredTimer _returnvalue_5))))))

(event Custom|TriggerBoredFace
  (bind _returnvalue (CustomCharacter|GetSkeletalMeshForPartOnPlayerPawn))
  ;[1 1 1] Bored
  (Rendering|Material|SetScalarParameterValueonMaterials _returnvalue "Hide Element 03" 0.0)
  (Class|FortCosmeticMaterialBlueprintLibrary|SetScalarParameterValueonOverlayMaterial _returnvalue "Hide Element 03")
  (Utilities|FlowControl|RetriggerableDelay 2.0)
  (Rendering|Material|SetScalarParameterValueonMaterials _returnvalue "Hide Element 03" 1.0)
  (Class|FortCosmeticMaterialBlueprintLibrary|SetScalarParameterValueonOverlayMaterial _returnvalue "Hide Element 03" 1.0))

(fn GameTimeFromBPForGlow
  (params (A :type Float%20%28double-precision%29) (B :type Float%20%28double-precision%29))
  (return 0.0))

(fn ConstructionScript)
