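; Cosmetic part modifier Blueprint for the Slurp variant of the Fish Head character.
; After the post-res-in event runs its parent logic, it gets the skeletal mesh for its part on the player pawn and sets that mesh's translucent sort priority to a negative value.
; The negative priority makes the translucent part draw behind other translucent objects.
; The construction script only calls the parent.
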(event EventOnPostResIn
  (|CallParentFunction)
  (Rendering|SetTranslucentSortPriority (CustomCharacter|GetSkeletalMeshForPartOnPlayerPawn) -50))

(fn ConstructionScript
  (|CallParentFunction))
