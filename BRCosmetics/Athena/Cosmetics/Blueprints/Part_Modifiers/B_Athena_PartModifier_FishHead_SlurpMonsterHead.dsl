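; Part modifier Blueprint for the FishHead Slurp Monster head cosmetic, adjusting render order of that character part on the player pawn.
; After the post-res-in event and its parent logic, it sets the translucent sort priority of the part's skeletal mesh to -1.
; The construction script only calls the parent implementation.
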(event EventOnPostResIn
  (|CallParentFunction)
  (Rendering|SetTranslucentSortPriority (CustomCharacter|GetSkeletalMeshForPartOnPlayerPawn) -1))

(fn ConstructionScript
  (|CallParentFunction))
