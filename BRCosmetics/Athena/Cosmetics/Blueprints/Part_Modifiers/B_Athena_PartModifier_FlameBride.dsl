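; FlameBride cosmetic part modifier that drives head flame materials and an idle Niagara effect on the player mesh.
; In frontend display, body angular-velocity changes set Ember and Burn targets, interpolated with retriggerable delays.
; In game, those parameters ramp up while moving outside skydiving and otherwise fade.
; The effect switches off while DBNO, and restarts with physics off if a skin weight profile is active.
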(event EventOnPostResIn
  (bind _returnvalue (CustomCharacter|GetAssociatedPlayerPawn))
  (bind _returnvalue_2 (Customization|GetSkeletalMeshforPartType _returnvalue "Head" :!class /Script/FortniteGame.FortPlayerPawn))
  (|CallParentFunction)
  (Utilities|FlowControl|Sequence
    (:then_0
      (bind _asniagara_particle_system_component (Utilities|Casting|CastToNiagaraComponent (CallFunction|GetIdleVFXComponent)))
      (Niagara|SetNiagaraSkeletalMeshComponent _asniagara_particle_system_component "PlayerSkeletalMesh" (Variables|/Script/Engine.Character|GetMesh (CustomCharacter|GetAssociatedPlayerPawn))))
    (:then_1
      (if (Reactive|Frontend|IsCosmeticObjectDisplayedinFrontendContext 0)
        (Rendering|Material|SetScalarParameterValueonMaterials _returnvalue_2 "Velocity Intensity Max" 2.0)
        (Rendering|Material|SetScalarParameterValueonMaterials _returnvalue_2 "Flame Smoothness Max" 1.1)
        (Utilities|FlowControl|Sequence
          (:then_0
            (Variables|Velocity|SetLast%20Position (Transformation|GetWorldLocation (Customization|GetSkeletalMeshforPartType (CustomCharacter|GetAssociatedPlayerPawn) "Body" :!class /Script/FortniteGame.FortPlayerPawn)))
            (Variables|Velocity|SetLast%20Linear%20Velocity "0, 0, 0")
            (Variables|Velocity|SetLast%20Rotation "0, 0, 0"))
          (:then_1
            (CallFunction|Started)))))
    (:then_2
      (EventDispatchers|BindEventtoOnDBNOStateChanged (CustomCharacter|GetAssociatedPlayerPawn) (delegate Custom|DBNOStateChange)))
    (:then_3
      (if (Components|SkinnedMesh|IsUsingSkinWeightProfile (Customization|GetSkeletalMeshforPartType (CustomCharacter|GetAssociatedPlayerPawn) (CustomCharacter|GetCharacterPartType) :!class /Script/FortniteGame.FortPlayerPawn))
        (bind _asniagara_particle_system_component_2 (Utilities|Casting|CastToNiagaraComponent (CallFunction|GetIdleVFXComponent)))
        (Niagara|SetNiagaraVariable(Bool) _asniagara_particle_system_component_2 "Change_OffPhysics" true)
        (Niagara|ResetSystem _asniagara_particle_system_component_2)
        (Components|Activation|SetActive _asniagara_particle_system_component_2 true true)))))

(event Custom|DBNOStateChange
  (params (FortPawn :type Fort%20Pawn%20Object%20Reference) (bInIsDBNO :type Boolean))
  (bind _returnvalue (CallFunction|GetIdleVFXComponent))
  (if bInIsDBNO
    (Components|Activation|Deactivate _returnvalue)
    (else
      (Components|Activation|Activate _returnvalue false))))

(event Custom|Started
  (bind (_target_low_high_x _target_low_high_y) (Math|Vector2D|BreakVector2D (Variables|GetTarget%20Low/High_X)))
  (bind (_target_low_high_x_2 _target_low_high_y_2) (Math|Vector2D|BreakVector2D (Variables|GetTarget%20Low/High_X)))
  (if (== (Variables|Default|GetTarget) _target_low_high_y)
    (Variables|Default|SetEmber%20Interp%20Speed 0.7)
    (Variables|Default|SetBurn%20Interp%20Speed 0.2)
    (Variables|Default|SetTarget _target_low_high_x_2)
    (Variables|Default|SetTarget%20Burn _target_low_high_x_2)
    (Utilities|FlowControl|RetriggerableDelay (Variables|Default|GetDelay%20Time)))
  (Variables|Default|SetEmber%20Interp%20Speed 1.0)
  (Variables|Default|SetBurn%20Interp%20Speed 1.0)
  (Variables|Default|SetTarget _target_low_high_y_2)
  (Variables|Default|SetTarget%20Burn _target_low_high_y_2))

(event EventOnVelocityToggle
  (params (IsMoving :type Boolean) (Delta%20Seconds :type Float%20%28double-precision%29))
  (bind _returnvalue (CustomCharacter|GetAssociatedPlayerPawn))
  (bind _returnvalue_2 (Customization|GetSkeletalMeshforPartType _returnvalue "Head" :!class /Script/FortniteGame.FortPlayerPawn))
  (bind _returnvalue_3 (CustomCharacter|GetAssociatedPlayerPawn))
  (bind _returnvalue_4 (Customization|GetSkeletalMeshforPartType _returnvalue_3 "Head" :!class /Script/FortniteGame.FortPlayerPawn))
  (bind (_target_low_high_x _target_low_high_y) (Math|Vector2D|BreakVector2D (Variables|GetTarget%20Low/High_X)))
  (bind (_target_low_high_x_2 _target_low_high_y_2) (Math|Vector2D|BreakVector2D (Variables|GetTarget%20Low/High_X)))
  (bind _current_burn_intensity (Variables|Default|GetCurrent%20Burn%20Intensity%20))
  (bind _lastangularvelocity (Variables|Velocity|GetLastAngularVelocity))
  (bind _current_flame_intensity (Variables|Default|GetCurrent%20Flame%20Intensity))
  (if (Reactive|Frontend|IsCosmeticObjectDisplayedinFrontendContext 0)
    (Utilities|FlowControl|Sequence
      (:then_0
        (bind _output_get (Variables|Default|SetCurrent (Math|Interpolation|FInterpTo (Variables|Default|GetCurrent) (Variables|Default|GetTarget) Delta%20Seconds (Variables|Default|GetEmber%20Interp%20Speed))))
        (bind _output_get_2 (Variables|Default|SetCurrent%20Burn (Math|Interpolation|FInterpTo (Variables|Default|GetCurrent%20Burn) (Variables|Default|GetTarget%20Burn) Delta%20Seconds (Variables|Default|GetBurn%20Interp%20Speed))))
        (Rendering|Material|SetScalarParameterValueonMaterials _returnvalue_4 "Ember Time" _output_get)
        (Rendering|Material|SetScalarParameterValueonMaterials _returnvalue_4 "Burn Time" _output_get_2))
      (:then_1
        (bind _angular_velocity (CallFunction|AngularVelocity Delta%20Seconds))
        (Variables|Velocity|SetLastAngularVelocity _angular_velocity)
        (if (and (Utilities|Operators|NotEqual(!=) :A _lastangularvelocity) (Variables|Velocity|GetVelocity%20Changed))
          (Variables|Default|SetCache (Variables|Velocity|GetLastAngularVelocity))
          (Variables|Velocity|SetVelocity%20Changed false)
          (Variables|Default|SetDelay%20Time 1.0)
          (Variables|Default|SetTarget _target_low_high_y_2)
          (Variables|Default|SetTarget%20Burn _target_low_high_y_2)
          (Utilities|FlowControl|RetriggerableDelay (Variables|Default|GetDelay%20Time))
          (if (== (Variables|Default|GetTarget) _target_low_high_y)
            (Variables|Default|SetEmber%20Interp%20Speed 0.7)
            (Variables|Default|SetBurn%20Interp%20Speed 0.2)
            (Variables|Default|SetTarget _target_low_high_x_2)
            (Variables|Default|SetTarget%20Burn _target_low_high_x_2))
          (Variables|Default|SetEmber%20Interp%20Speed 1.0)
          (Variables|Default|SetBurn%20Interp%20Speed 1.0)
          (elif (Utilities|Operators|NotEqual(!=) _lastangularvelocity (Variables|Default|GetCache) 0)
            (Variables|Velocity|SetVelocity%20Changed true)
            (Variables|Default|SetDelay%20Time 15.0)))))
    (else
      (Utilities|IsSkydiving (CustomCharacter|GetAssociatedPlayerPawn)
        (:true
          _)
        (:false
          (if IsMoving
            (bind _output_get_3 (Variables|Default|SetCurrent%20Flame%20Intensity (Math|Interpolation|FInterpTo _current_flame_intensity 1.0 Delta%20Seconds 0.5)))
            (bind _output_get_4 (Variables|Default|SetCurrent%20Burn%20Intensity%20 (Math|Interpolation|FInterpTo _current_burn_intensity 1.0 Delta%20Seconds 0.5)))
            (Rendering|Material|SetScalarParameterValueonMaterials _returnvalue_2 "Ember Time" _output_get_3)
            (Rendering|Material|SetScalarParameterValueonMaterials _returnvalue_2 "Burn Time" _output_get_4)
            (else
              _))))
      (bind _output_get_5 (Variables|Default|SetCurrent%20Flame%20Intensity (Math|Interpolation|FInterpTo _current_flame_intensity 0.0 Delta%20Seconds (Variables|Default|GetEmber%20Interp%20Speed))))
      (bind _output_get_6 (Variables|Default|SetCurrent%20Burn%20Intensity%20 (Math|Interpolation|FInterpTo _current_burn_intensity 0.0 Delta%20Seconds (Variables|Default|GetBurn%20Interp%20Speed))))
      (Rendering|Material|SetScalarParameterValueonMaterials _returnvalue_2 "Ember Time" _output_get_5)
      (Rendering|Material|SetScalarParameterValueonMaterials _returnvalue_2 "Burn Time" _output_get_6))))

(fn AngularVelocity
  (params (Delta%20Time :type Float%20%28double-precision%29))
  (bind _returnvalue (CustomCharacter|GetAssociatedPlayerPawn))
  (bind _returnvalue_2 (Customization|GetSkeletalMeshforPartType _returnvalue "Body" :!class /Script/FortniteGame.FortPlayerPawn))
  (bind (_returnvalue_roll _returnvalue_pitch _returnvalue_yaw) (Math|Rotator|BreakRotator (Transformation|GetWorldRotation _returnvalue_2)))
  (bind _returnvalue_3 (CustomCharacter|GetAssociatedPlayerPawn))
  (bind _returnvalue_4 (Customization|GetSkeletalMeshforPartType _returnvalue_3 "Body" :!class /Script/FortniteGame.FortPlayerPawn))
  (bind _returnvalue_5 (Transformation|GetWorldLocation _returnvalue_4))
  (bind _last_position (Variables|Velocity|GetLast%20Position))
  (bind _returnvalue_6 (- _returnvalue_5 _last_position))
  (bind _returnvalue_7 (/ _returnvalue_6 Delta%20Time))
  (bind _last_rotation (Variables|Velocity|GetLast%20Rotation))
  (bind _returnvalue_8 (Utilities|Operators|Subtract (Math|Vector|MakeVector _returnvalue_roll _returnvalue_pitch _returnvalue_yaw) _last_rotation))
  (bind _returnvalue_9 (/ _returnvalue_8 Delta%20Time))
  (Variables|Velocity|SetLinear%20Acceleration (* (* (/ (- _returnvalue_7 (Variables|Velocity|GetLast%20Linear%20Velocity)) Delta%20Time) 0.01) (Variables|Velocity|GetLinear%20Scale%20Multiplier))) ; Acceleration in m/s
  (Variables|Velocity|SetLast%20Linear%20Velocity _returnvalue_7)
  (Variables|Velocity|SetLast%20Position _returnvalue_5)
  ;[1 1 1] Angular Acceleration
  (Variables|Velocity|SetAngular%20Acceleration (Math|Vector|VectortoRadians (* (/ (- _returnvalue_9 (Variables|Velocity|GetLastAngularVelocity)) Delta%20Time) (Variables|Velocity|GetAngular%20Scale%20Multiplier))))
  (bind _output_get (Variables|Velocity|SetLastAngularVelocity _returnvalue_9))
  (Variables|SetLast%20Rotation_X (Math|Vector|MakeVector _returnvalue_roll _returnvalue_pitch _returnvalue_yaw))
  (return _output_get))

(fn ConstructionScript
  (|CallParentFunction))
