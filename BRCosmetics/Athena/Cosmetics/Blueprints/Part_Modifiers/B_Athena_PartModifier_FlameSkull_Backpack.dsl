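; Part modifier for the Flame Skull backpack cosmetic that plays an effect when the pawn jumps.
; On the pawn jump event it runs the parent behavior, then spawns the backpack's jump Niagara system.
; The system attaches to the FXSocket on the skeletal mesh of the cosmetic part on the player pawn.
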(event EventOnPawnJumped
  (|CallParentFunction)
  (PartModifier|SpawnSystem|SpawnSystemAttachedPartModifier "/BRCosmetics/Accessories/FORT_Backpacks/Backpack_M_MED_Flame_Skull/FX/NS_Backpack_FlameSkull_Jump.NS_Backpack_FlameSkull_Jump" (CustomCharacter|GetSkeletalMeshForPartOnPlayerPawn) "FXSocket"))

(fn ConstructionScript)
