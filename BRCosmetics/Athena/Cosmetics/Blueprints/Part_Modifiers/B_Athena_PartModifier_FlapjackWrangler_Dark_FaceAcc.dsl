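; Cosmetic part modifier for the dark Flapjack Wrangler face accessory that drives its idle and landing effects.
; On each character movement update, it feeds the pawn's IsMovingOnGround result into the idle VFX component's IsOnGround parameter.
; On landing, it waits briefly, then spawns the Niagara landing effect attached to the part's skeletal mesh with no socket, passing that mesh's world location.
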(event EventOnCharacterMovementUpdated
  (params (DeltaSeconds :type Float%20%28single-precision%29) (OldLocation :type Vector :default "0, 0, 0") (OldVelocity :type Vector :default "0, 0, 0"))
  (Effects|Components|ParticleSystem|SetBoolParameter (CallFunction|GetIdleVFXComponent) "IsOnGround" (AI|Components|NavMovement|IsMovingonGround (Variables|/Script/Engine.Character|GetCharacterMovement (CustomCharacter|GetAssociatedPlayerPawn)))))

(event EventOnPawnLanded
  (params (Hit :type Hit%20Result%20Structure%20%28by%20ref%29))
  (Utilities|FlowControl|Delay 0.125)
  (PartModifier|SpawnSystem|SpawnSystemAttachedPartModifier "/BRCosmetics/Characters/Player/Male/Medium/Bodies/M_MED_Flapjack_Wrangler_Dark_Updates/FX/NS_M_Med_FW_Dark_Updates_Land.NS_M_Med_FW_Dark_Updates_Land" (CustomCharacter|GetSkeletalMeshForPartOnPlayerPawn) "None" (Transformation|GetWorldLocation (CustomCharacter|GetSkeletalMeshForPartOnPlayerPawn))))

(fn ConstructionScript)
