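; Cosmetic part modifier for the Flapjack Wrangler face accessory that drives its idle and landing VFX on the player pawn.
; Each movement update feeds the pawn's on-ground state into the idle VFX component's IsOnGround parameter.
; On landing, unless the head is hidden, it waits briefly, then spawns a Niagara burst attached to the part's skeletal mesh.
; The Progressive variant stage picks that burst: early stages share the base effect, later ones use Saphire, Topaz or ZeroPoint versions.
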(event EventOnCharacterMovementUpdated
  (params (DeltaSeconds :type Float%20%28single-precision%29) (OldLocation :type Vector :default "0, 0, 0") (OldVelocity :type Vector :default "0, 0, 0"))
  (Effects|Components|ParticleSystem|SetBoolParameter (CallFunction|GetIdleVFXComponent) "IsOnGround" (AI|Components|NavMovement|IsMovingonGround (Variables|/Script/Engine.Character|GetCharacterMovement (CustomCharacter|GetAssociatedPlayerPawn)))))

(event EventOnPawnLanded
  (params (Hit :type Hit%20Result%20Structure%20%28by%20ref%29))
  (if (not (Variables|HiddenPartCheck|GetbIsHeadHidden))
    (Utilities|FlowControl|Delay 0.125)
    (bind _index (CosmeticLoadout|GetVariantChannelActiveIndexforSlotActor "(TagName=\"LoadoutSlot.Character\")" (CustomCharacter|GetAssociatedPlayerPawn) "(TagName=\"Cosmetics.Variant.Channel.Progressive\")" (Utilities|Array|MakeArray "(TagName=\"Cosmetics.Variant.Property.Stage1\")" "(TagName=\"Cosmetics.Variant.Property.Stage2\")" "(TagName=\"Cosmetics.Variant.Property.Stage3\")" "(TagName=\"Cosmetics.Variant.Property.Stage4\")" "(TagName=\"Cosmetics.Variant.Property.Stage5\")" "(TagName=\"Cosmetics.Variant.Property.Stage6\")" "(TagName=\"Cosmetics.Variant.Property.Stage7\")")))
    (PartModifier|SpawnSystem|SpawnSystemAttachedPartModifier (Utilities|Select :Index _index :Option%200 "/BRCosmetics/Characters/Player/Male/Medium/Bodies/M_MED_Flapjack_Wrangler/FX/NS_M_Med_Flapjack_Wrangler_Land.NS_M_Med_Flapjack_Wrangler_Land" :Option%201 "/BRCosmetics/Characters/Player/Male/Medium/Bodies/M_MED_Flapjack_Wrangler/FX/NS_M_Med_Flapjack_Wrangler_Land.NS_M_Med_Flapjack_Wrangler_Land" :Option%202 "/BRCosmetics/Characters/Player/Male/Medium/Bodies/M_MED_Flapjack_Wrangler/FX/NS_M_Med_Flapjack_Wrangler_Land.NS_M_Med_Flapjack_Wrangler_Land" :Option%203 "/BRCosmetics/Characters/Player/Male/Medium/Bodies/M_MED_Flapjack_Wrangler/FX/NS_M_Med_Flapjack_Wrangler_Land.NS_M_Med_Flapjack_Wrangler_Land" :Option%204 "/BRCosmetics/Characters/Player/Male/Medium/Bodies/M_MED_Flapjack_Wrangler/FX/NS_M_Med_FW_Land_Saphire.NS_M_Med_FW_Land_Saphire" :Option%205 "/BRCosmetics/Characters/Player/Male/Medium/Bodies/M_MED_Flapjack_Wrangler/FX/NS_M_Med_FW_Land_Topaz.NS_M_Med_FW_Land_Topaz" :Option%206 "/BRCosmetics/Characters/Player/Male/Medium/Bodies/M_MED_Flapjack_Wrangler/FX/NS_M_Med_FW_Land_ZeroPoint.NS_M_Med_FW_Land_ZeroPoint") (CustomCharacter|GetSkeletalMeshForPartOnPlayerPawn) "None" (Transformation|GetWorldLocation (CustomCharacter|GetSkeletalMeshForPartOnPlayerPawn)))))

(fn ConstructionScript)
