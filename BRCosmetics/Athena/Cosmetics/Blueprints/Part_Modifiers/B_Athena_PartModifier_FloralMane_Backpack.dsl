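; Part modifier for the FloralMane backpack cosmetic that drives its movement-reactive, variant-aware effects.
; On post res-in it caches the player pawn and maps the backpack's Progressive variant stage to a 0 or 1 Variant float.
; On landing it spawns the Jump Niagara system at the mesh root and passes that float.
; Outside frontend display, Tick sets the idle VFX sprint flag while the pawn slides or tactical sprints.
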(event EventOnPostResIn
  (|CallParentFunction)
  (bind _asfort_player_pawn_athena (Utilities|Casting|CastToFortPlayerPawnAthena (CustomCharacter|GetAssociatedPlayerPawn)))
  (Variables|Default|SetplayerRef _asfort_player_pawn_athena)
  (bind _index (CosmeticLoadout|GetVariantChannelActiveIndexforSlotActor "(TagName=\"LoadoutSlot.Backpack\")" (CustomCharacter|GetAssociatedPlayerPawn) "(TagName=\"Cosmetics.Variant.Channel.Progressive\")" (Utilities|Array|MakeArray "(TagName=\"Cosmetics.Variant.Property.Stage1\")" "(TagName=\"Cosmetics.Variant.Property.Stage2\")")))
  (if (== _index 0)
    (Variables|Default|SetVariant 0.0)
    (else
      (Variables|Default|SetVariant 1.0))))

(event EventOnDayPhaseChanged
  (params (CurrentDayPhase :type EFortDayPhase%20Enum :default "Morning") (PreviousDayPhase :type EFortDayPhase%20Enum :default "Morning") (bAtCreation :type Boolean)))

(event EventOnNewPlayerWeaponType
  (params (NewType :type EFortItemType%20Enum :default "WorldItem") (OldType :type EFortItemType%20Enum :default "WorldItem"))
  (|CallParentFunction NewType OldType) ; This node is disabled and will not be called.\nDrag off pins to build functionality.
)

(event EventOnWeaponChanged
  (params (NewWeapon :type Fort%20Weapon%20Object%20Reference) (OldWeapon :type Fort%20Weapon%20Object%20Reference))
  (|CallParentFunction NewWeapon OldWeapon) ; This node is disabled and will not be called.\nDrag off pins to build functionality.
)

(event EventOnWeaponFired)

(event EventOnNewWeapons
  (params (WeaponsAsItems :type Array%20of%20Fort%20Item%20Object%20References)))

(event EventOnCurieStateAttached
  (params (Tag :type Gameplay%20Tag%20Structure%20%28by%20ref%29)))

(event EventOnCurieStateDetached
  (params (Tag :type Gameplay%20Tag%20Structure%20%28by%20ref%29)))

(event EventOnBeginSkydiving
  (|CallParentFunction) ; This node is disabled and will not be called.\nDrag off pins to build functionality.
)

(event EventOnEndSkydiving
  (|CallParentFunction) ; This node is disabled and will not be called.\nDrag off pins to build functionality.
)

(event EventOnEnterVehicle
  (params (VehicleEntered :type Fort%20Vehicle%20Interface%20Interface%20%28by%20ref%29))
  (|CallParentFunction VehicleEntered) ; This node is disabled and will not be called.\nDrag off pins to build functionality.
)

(event EventOnExitVehicle
  (|CallParentFunction) ; This node is disabled and will not be called.\nDrag off pins to build functionality.
)

(event EventOnTargetingOrFirstPersonCameraChanged)

(event EventOnHitPawn
  (params (SelfActor :type Actor%20Object%20Reference) (OtherActor :type Actor%20Object%20Reference) (NormalImpulse :type Vector :default "0, 0, 0") (Hit :type Hit%20Result%20Structure%20%28by%20ref%29)))

(event EventOnPawnLanded
  (|CallParentFunction)
  (bind _returnvalue (PartModifier|SpawnSystem|SpawnSystemAttachedPartModifier "/BRCosmetics/Accessories/FORT_Backpacks/Backpack_M_MED_FloralMane/FX/NS_Backpack_FloralMane_Jump.NS_Backpack_FloralMane_Jump" (CustomCharacter|GetSkeletalMeshForPartOnPlayerPawn) "root" "0, 0, 0" "0, 0, 0" "SnapToTarget")
    (:Success
      (Niagara|SetNiagaraVariableByString(Float) _returnvalue "Variant" (Variables|Default|GetVariant)))))

(event EventTick
  (params (DeltaSeconds :type Float%20%28single-precision%29))
  (if (not (Reactive|Frontend|IsCosmeticObjectDisplayedinFrontendContext 0))
    (Effects|Components|ParticleSystem|SetBoolParameter (CallFunction|GetIdleVFXComponent) "bIsSprinting" (or (Variables|/Script/FortniteGame.FortPawn|GetbIsSliding (Variables|Default|GetplayerRef)) (Variables|/Script/FortniteGame.FortPlayerPawn|GetbIsTacticalSprinting (Variables|Default|GetplayerRef))))))

(fn ConstructionScript
  (|CallParentFunction))
