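; Character part modifier for the FoundKelp Cast cosmetic that handles its ice body rendering and kill reactivity.
; Customization applies an opaque ice material on mobile, Switch and PC performance mode; otherwise it enables custom depth for transparency, re-asserted by Tick during the first second.
; After res-in, watched kill stat changes, or a simulated frontend count, feed game time to the material and timeline-driven curve and KillCount to idle VFX.
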(event EventOnCharacterCustomizationFinished
  (params (CharacterPartInfo :type Array%20of%20Applied%20Character%20Part%20Info%20Structures))
  (bind _returnvalue (CustomCharacter|GetSkeletalMeshForPartOnPlayerPawn))
  (|CallParentFunction CharacterPartInfo)
  ; set to masked material on mobile and low ends to avoid sorting issue
  (if (or (or (Fortnite|IsMobilePlatform) (Class|PlatformHelpersLibrary|IsPcPerfMode)) (TouchPlatform|IsNintendoSwitch))
    (Rendering|Material|SetMaterial _returnvalue 2 "/BRCosmetics/Characters/Player/Male/Medium/Bodies/M_MED_FoundKelp_Cast/Materials/MI_FoundKelp_Cast_Body_Ice_Opaque.MI_FoundKelp_Cast_Body_Ice_Opaque")
    (else
      ; Turn on render custom depth for ice transparency
      (Rendering|SetRenderCustomDepth _returnvalue true)
      (Variables|Default|SetRender%20Custom%20Depth%20Pass%20On true))))

(event EventTick
  (params (DeltaSeconds :type Float%20%28single-precision%29))
  (bind _returnvalue (CustomCharacter|GetSkeletalMeshForPartOnPlayerPawn))
  ; Toggle custom depth on if it got overrided
  (|CallParentFunction DeltaSeconds)
  (if (> (Actor|GetGameTimeSinceCreation) 1.0)
    (Actor|Tick|SetActorTickEnabled false)
    (elif (and (== (Variables|/Script/Engine.PrimitiveComponent|GetbRenderCustomDepth _returnvalue) false) (Variables|Default|GetRender%20Custom%20Depth%20Pass%20On))
      (Rendering|SetRenderCustomDepth _returnvalue true))))

(event EventOnPostResIn
  (bind _returnvalue (CustomCharacter|GetSkeletalMeshForPartOnPlayerPawn))
  (bind _material_instance_dynamic (Class|BTechArtGlobalLibrary|GetDynamicMaterialInstance _returnvalue 1))
  (bind _returnvalue_2 (CallFunction|GetIdleVFXComponent))
  (bind _returnvalue_3 (Utilities|Time|GetGameTimeinSeconds))
  (bind _fe_killcount (Variables|Default|GetFE_KillCount))
  (|CallParentFunction)
  (if (CallFunction|ShouldDisplayFrontendReactivity)
    ; FE preview
    (Rendering|Material|SetScalarParameterValue _material_instance_dynamic "FKC_GameTimeFromBP" _returnvalue_3 :!class /Script/Engine.MaterialInstanceDynamic)
    (bind (_direction _newtrack_0) (Timeline|Timeline_0 PlayFromStart)
      (:Update
        (Effects|Components|ParticleSystem|SetFloatParameter _returnvalue_2 "ReactiveCurve" _newtrack_0)
        (Effects|Components|ParticleSystem|SetIntParameter _returnvalue_2 "KillCount" _fe_killcount)
        (Utilities|FlowControl|Delay 9.0)
        (Variables|Default|SetFE_KillCount (+ 1 _fe_killcount))))
    (else
      (bind (_statname _statvalue) (Cosmetics|WatchCosmeticStat (CustomCharacter|GetAssociatedPlayerPawn) "(TagName=\"StatManager.AthenaKills\")" 0 false)
        (:OnStatChanged
          (Effects|Components|ParticleSystem|SetIntParameter (CallFunction|GetIdleVFXComponent) "KillCount" _statvalue)
          ;[1 1 1] In game
          (Rendering|Material|SetScalarParameterValue _material_instance_dynamic "FKC_GameTimeFromBP" _returnvalue_3 :!class /Script/Engine.MaterialInstanceDynamic)
          (bind (_direction_2 _newtrack_0_2) (Timeline|Timeline PlayFromStart)
            (:Update
              (Effects|Components|ParticleSystem|SetFloatParameter _returnvalue_2 "ReactiveCurve" _newtrack_0_2))))))))

(fn ConstructionScript
  (|CallParentFunction))
