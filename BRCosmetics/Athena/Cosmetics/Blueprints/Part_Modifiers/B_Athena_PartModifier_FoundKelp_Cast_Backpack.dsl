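; Cosmetic part modifier for the FoundKelp Cast backpack that timestamps its material when the player's kill stat changes.
; After res-in it gets a dynamic material from the part's skeletal mesh and writes game time to FKC_GameTimeFromBP.
; In the front end it writes this once as a preview, followed by a delay with nothing after it.
; In game it watches StatManager.AthenaKills and writes on each change.
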(event EventOnPostResIn
  (bind _returnvalue (CustomCharacter|GetSkeletalMeshForPartOnPlayerPawn))
  (bind _material_instance_dynamic (Class|BTechArtGlobalLibrary|GetDynamicMaterialInstance _returnvalue))
  (bind _returnvalue_2 (Utilities|Time|GetGameTimeinSeconds))
  (|CallParentFunction)
  (if (CustomCosmeticModifier|IsinFrontEnd)
    ; FE preview
    (Rendering|Material|SetScalarParameterValue _material_instance_dynamic "FKC_GameTimeFromBP" _returnvalue_2 :!class /Script/Engine.MaterialInstanceDynamic)
    (Utilities|FlowControl|Delay 8.0)
    (else
      (Cosmetics|WatchCosmeticStat (CustomCharacter|GetAssociatedPlayerPawn) "(TagName=\"StatManager.AthenaKills\")" 0 false
        (:OnStatChanged
          ;[1 1 1] In game
          (Rendering|Material|SetScalarParameterValue _material_instance_dynamic "FKC_GameTimeFromBP" _returnvalue_2 :!class /Script/Engine.MaterialInstanceDynamic))))))

(fn ConstructionScript
  (|CallParentFunction))
