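; Cosmetic part modifier for the Fox Warrior skin that plays a Niagara burst effect on the character.
; After res-in, it caches the active Parts variant stage index and, in the front end, triggers the burst after a short delay.
; In play, any stat change with a positive value spawns the same root-attached burst after a brief delay.
; Both paths pass the cached index to the effect as VariantNum.
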(event EventOnPostResIn
  (|CallParentFunction)
  (bind _index (CosmeticLoadout|GetVariantChannelActiveIndexforSlotActor "(TagName=\"LoadoutSlot.Character\")" (CustomCharacter|GetAssociatedPlayerPawn) "(TagName=\"Cosmetics.Variant.Channel.Parts\")" (Utilities|Array|MakeArray "(TagName=\"Cosmetics.Variant.Property.Stage1\")" "(TagName=\"Cosmetics.Variant.Property.Stage2\")")))
  (Variables|Default|SetVariant%20Index _index)
  (if (CustomCosmeticModifier|IsinFrontEnd)
    (Utilities|FlowControl|Delay 0.75)
    (CallFunction|OnKillRiggerFE)
    (Utilities|FlowControl|Delay 4.0)))

(event Custom|OnKillRigger_FE
  (Utilities|FlowControl|Delay 0.125)
  (bind _returnvalue (PartModifier|SpawnSystem|SpawnSystemAttachedPartModifier "/BRCosmetics/Characters/Player/Female/Medium/Bodies/F_MED_Fox_Warrior/Effects/NS_Skin_FoxWarrior_Burst.NS_Skin_FoxWarrior_Burst" (CustomCharacter|GetSkeletalMeshForPartOnPlayerPawn) "root")
    (:Success
      (Effects|Components|ParticleSystem|SetIntParameter _returnvalue "VariantNum" (Variables|Default|GetVariant%20Index)))))

(event EventOnStatChanged
  (params (State%20Name :type Name :default "None") (Stat%20Value :type Integer))
  (|CallParentFunction State%20Name Stat%20Value)
  (if (> Stat%20Value 0)
    (Utilities|FlowControl|Delay 0.125)
    (bind _returnvalue (PartModifier|SpawnSystem|SpawnSystemAttachedPartModifier "/BRCosmetics/Characters/Player/Female/Medium/Bodies/F_MED_Fox_Warrior/Effects/NS_Skin_FoxWarrior_Burst.NS_Skin_FoxWarrior_Burst" (CustomCharacter|GetSkeletalMeshForPartOnPlayerPawn) "root")
      (:Success
        (Effects|Components|ParticleSystem|SetIntParameter _returnvalue "VariantNum" (Variables|Default|GetVariant%20Index))))))

(fn ConstructionScript
  (|CallParentFunction))
