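; Part modifier for the Funky Feather Coin cosmetic that flags the character's body materials when shown in the front end.
; On post-resurrection-in, it calls the parent, then checks whether it is in the front end.
; If so, it sets the IsInFE scalar parameter to 1 on all materials of the associated player pawn's Body skeletal mesh.
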(event EventOnPostResIn
  (|CallParentFunction)
  (if (CustomCosmeticModifier|IsinFrontEnd)
    (Rendering|Material|SetScalarParameterValueonMaterials (Customization|GetSkeletalMeshforPartType (CustomCharacter|GetAssociatedPlayerPawn) "Body" :!class /Script/FortniteGame.FortPlayerPawn) "IsInFE" 1.0)))

(fn ConstructionScript
  (|CallParentFunction))
