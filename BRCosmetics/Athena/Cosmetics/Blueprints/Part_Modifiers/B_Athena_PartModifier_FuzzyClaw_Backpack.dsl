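; Cosmetic part modifier for the FuzzyClaw backpack that makes its face react to player events through material expressions.
; After res-in, it either runs a delay-timed frontend preview or watches the owning pawn's DamageReceived and OpenChests stats.
; Reactions write an expression index (sad for damage, sparkle eyes for chest opening) and a game-time value to the part mesh's materials.
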(event EventOnPostResIn
  (bind _returnvalue (CustomCharacter|GetSkeletalMeshForPartOnPlayerPawn))
  (bind _returnvalue_2 (CustomCharacter|GetAssociatedPlayerPawn))
  (bind _returnvalue_3 (Utilities|Time|GetGameTimeinSeconds))
  (|CallParentFunction)
  (if (Reactive|Frontend|ShouldDisplayFrontendReactivity 0 "Backpack")
    ; FE preview cycle
    (Utilities|FlowControl|Delay 2.5)
    (Utilities|FlowControl|Sequence
      (:then_0
        (Utilities|FlowControl|Delay 7.5)
        (Utilities|FlowControl|Sequence
          (:then_0
            (Utilities|FlowControl|Delay 5.0))
          (:then_1
            _))
        ; In game reactivity
        (Rendering|Material|SetScalarParameterValueonMaterials _returnvalue "FC_GameTimeFromBP" _returnvalue_3)
        (Rendering|Material|SetScalarParameterValueonMaterials _returnvalue "FC_ExpressionInt" 0.0 :!pinned true) ; On damage: sad face
      )
      (:then_1
        _))
    (Rendering|Material|SetScalarParameterValueonMaterials _returnvalue "FC_GameTimeFromBP" _returnvalue_3)
    (Rendering|Material|SetScalarParameterValueonMaterials _returnvalue "FC_ExpressionInt" 1.0 :!pinned true) ; On chest opening: sparkle eyes
    (else
      (Utilities|FlowControl|Sequence
        (:then_0
          (Cosmetics|WatchCosmeticStat _returnvalue_2 "(TagName=\"StatManager.DamageReceived\")" 0 false
            (:OnStatChanged
              _)))
        (:then_1
          (Cosmetics|WatchCosmeticStat _returnvalue_2 "(TagName=\"StatManager.OpenChests\")" 0 false
            (:OnStatChanged
              _)))))))

(fn ConstructionScript
  (|CallParentFunction))
