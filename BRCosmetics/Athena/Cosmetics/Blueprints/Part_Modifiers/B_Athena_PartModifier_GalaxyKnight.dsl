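; Galaxy Knight cosmetic part modifier that makes the character part's emissive glow and idle hair VFX react to gameplay.
; Timelines drive material and particle parameters, replaying a brightening pulse with a seven-second delay in the front end.
; In matches, watched kill stats trigger brightening and a sound, while damage dims the material unless the pawn is DBNO.
; The DBNO delegate dims the glow and hides the hair, restoring both when that state clears.
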(event EventOnPostResIn
  (bind _returnvalue (CustomCharacter|GetSkeletalMeshForPartOnPlayerPawn))
  (bind _returnvalue_2 (CustomCharacter|GetAssociatedPlayerPawn))
  (|CallParentFunction)
  (if (CustomCosmeticModifier|IsinFrontEnd)
    ;[0.07005 0.2462 1] FrontEnd Reactivity Cycling
    (bind (_direction _brightening_onelim _brighteninghair_onelim) (Timeline|Timeline_2 PlayFromStart)
      (:Update
        (Rendering|Material|SetScalarParameterValueonMaterials (CustomCharacter|GetSkeletalMeshForPartOnPlayerPawn) "emissive mult" _brightening_onelim)
        (Effects|Components|ParticleSystem|SetFloatParameter (CallFunction|GetIdleVFXComponent) "EmissiveMult" _brighteninghair_onelim)
        (Utilities|FlowControl|Delay 7.0)
        (Timeline|Timeline_2 PlayFromStart)))
    (else
      ;[1 1 1] Damage Received
      (Utilities|FlowControl|Sequence
        (:then_0
          (EventDispatchers|BindEventtoOnDBNOStateChanged _returnvalue_2 (delegate Custom|DBNO)))
        (:then_1
          (Cosmetics|WatchCosmeticStat _returnvalue_2 "(TagName=\"StatManager.AthenaKills\")" 0 false
            (:OnStatChanged
              (Utilities|FlowControl|Sequence
                (:then_0
                  (bind (_direction_2 _brightening_onelim_2 _brighteninghair_onelim_2) (Timeline|Timeline_0 PlayFromStart)
                    (:Update
                      (Rendering|Material|SetScalarParameterValueonMaterials _returnvalue "emissive mult" _brightening_onelim_2)
                      (Effects|Components|ParticleSystem|SetFloatParameter (CallFunction|GetIdleVFXComponent) "EmissiveMult" _brighteninghair_onelim_2))))
                (:then_1
                  (Audio|SpawnSoundAttached (Variables|Default|GetOnKillSound) (CustomCharacter|GetSkeletalMeshForPartOnPlayerPawn)))))))
        (:then_2
          (bind (_statname _statvalue) (Cosmetics|WatchCosmeticStat _returnvalue_2 "(TagName=\"StatManager.DamageReceived\")" 0 false)
            (:OnStatChanged
              (if (> _statvalue 0)
                (if (not (stale Custom|DBNO bInIsDBNO))
                  (bind (_direction_3 _dimming_ondamage) (Timeline|Timeline_1 PlayFromStart)
                    (:Update
                      (Rendering|Material|SetScalarParameterValueonMaterials _returnvalue "emissive mult" _dimming_ondamage))))))))))))

(event Custom|DBNO
  (params (FortPawn :type Fort%20Pawn%20Object%20Reference) (bInIsDBNO :type Boolean))
  (bind _returnvalue (CustomCharacter|GetSkeletalMeshForPartOnPlayerPawn))
  (bind _returnvalue_2 (CallFunction|GetIdleVFXComponent))
  ;[1 1 1] Damage Received
  (bind (_direction _dimming _brightening _showhair _hidehair) (Timeline|Timeline PlayFromStart)
    (:Update
      (if bInIsDBNO
        (Rendering|Material|SetScalarParameterValueonMaterials _returnvalue "emissive mult" _dimming)
        (Effects|Components|ParticleSystem|SetFloatParameter _returnvalue_2 "OpacityMult" _hidehair)
        (else
          (Rendering|Material|SetScalarParameterValueonMaterials _returnvalue "emissive mult" _brightening)
          (Effects|Components|ParticleSystem|SetFloatParameter _returnvalue_2 "OpacityMult" _showhair))))))

(event Custom|Audio%20Setup%20&%20Activation)

(fn GetBasisDirection
  (params (Local%20Player%20Acceleration :type Vector))
  (bind _returnvalue (Math|Vector|Normalize Local%20Player%20Acceleration))
  (bind _x (.x _returnvalue))
  (bind _y (.y _returnvalue))
  (bind _returnvalue_2 (Math|Float|Absolute(Float) _y))
  (bind _returnvalue_3 (Math|Float|Absolute(Float) _x))
  (if (> _returnvalue_3 _returnvalue_2)
    (LocalVariables|Setmy_direction_amount _returnvalue_3)
    (if (>= _x 0.0)
      (LocalVariables|Setmy_direction 0)
      (else
        (LocalVariables|Setmy_direction 2)))
    (else
      (LocalVariables|Setmy_direction_amount _returnvalue_2)
      (if (>= _y 0.0)
        (LocalVariables|Setmy_direction 1)
        (else
          (LocalVariables|Setmy_direction 3)))))
  (return (LocalVariables|Getmy_direction) (LocalVariables|Getmy_direction_amount)))

(fn PlaySound
  (params (NewSound :type Sound%20Base%20Object%20Reference))
  (Audio|SpawnSoundAttached NewSound (CustomCharacter|GetSkeletalMeshForPartOnPlayerPawn)))

(fn ConstructionScript
  (|CallParentFunction))
