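; Character part modifier for the Galileo Gondola body cosmetic that animates a "TurnOn" material effect on the player.
; After res-in, it calls the parent handler and plays a timeline from the start.
; Each timeline update writes the "TurnOn" scalar to materials on the skeletal mesh components directly attached to the pawn's mesh.
; A custom TesTurnOn event also activates the idle VFX before playing the same timeline.
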(event Custom|TesTurnOn
  (Components|Activation|SetActive (CallFunction|GetIdleVFXComponent) true true)
  (bind (_direction _turnon) (Timeline|Timeline_0 PlayFromStart)
    (:Update
      (for _array_element (Components|GetChildrenComponents (Variables|/Script/Engine.Character|GetMesh (CustomCharacter|GetAssociatedPlayerPawn)) false)
        (bind _asskeletal_mesh_component (Utilities|Casting|CastToSkeletalMeshComponent _array_element))
        (Rendering|Material|SetScalarParameterValueonMaterials _asskeletal_mesh_component "TurnOn" _turnon)))))

(event EventOnPostResIn
  (|CallParentFunction)
  (Timeline|Timeline_0 PlayFromStart))

(fn ConstructionScript)
