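; Character part modifier for the Gatehound cosmetic that matches its idle VFX to the equipped progressive style stage.
; On the post-res-in event, it waits briefly via a retriggerable delay, then looks up which of three Progressive variant stages is active on the owning pawn's character loadout slot.
; It casts the idle VFX component to Niagara and passes that stage index in as the VariantNumber integer parameter.
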(event EventOnPostResIn
  (|CallParentFunction)
  (Utilities|FlowControl|RetriggerableDelay 0.2)
  (bind _index (CosmeticLoadout|GetVariantChannelActiveIndexforSlotActor "(TagName=\"LoadoutSlot.Character\")" (CustomCharacter|GetAssociatedPlayerPawn) "(TagName=\"Cosmetics.Variant.Channel.Progressive\")" (Utilities|Array|MakeArray "(TagName=\"Cosmetics.Variant.Property.Stage1\")" "(TagName=\"Cosmetics.Variant.Property.Stage2\")" "(TagName=\"Cosmetics.Variant.Property.Stage3\")")))
  (bind _asniagara_particle_system_component (Utilities|Casting|CastToNiagaraComponent (CallFunction|GetIdleVFXComponent)))
  (Effects|Components|ParticleSystem|SetIntParameter _asniagara_particle_system_component "VariantNumber" _index))

(fn ConstructionScript
  (|CallParentFunction))
