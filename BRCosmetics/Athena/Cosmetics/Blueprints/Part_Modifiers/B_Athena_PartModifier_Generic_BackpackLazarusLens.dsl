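; Cosmetic part modifier for the Lazarus Lens backpack that turns on idle VFX when parachute movement ends with a specific glider equipped.
; After res-in, it binds a handler to the associated player pawn's end-parachute-movement event.
; The handler compares the glider in the pawn's cosmetic loadout to the Glider ID variable and activates the idle VFX component on a match.
; The begin-parachute disable event is an empty stub.
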(event EventOnPostResIn
  (|CallParentFunction)
  (EventDispatchers|BindEventtoOnEndParachuteMovementEvent (CustomCharacter|GetAssociatedPlayerPawn) (delegate Custom|FX%20ON%20End%20Parachute%20Movement)))

(event Custom|Disable%20FX%20On%20Begin%20Parachute%20Movement)

(event Custom|FX%20ON%20End%20Parachute%20Movement
  (bind (_returnvalue_bannericonid _returnvalue_bannercolorid _returnvalue_skydivecontrail _returnvalue_glider) (Utilities|Struct|BreakFortAthenaLoadout (Cosmetics|GetPawnCosmeticLoadout (CustomCharacter|GetAssociatedPlayerPawn) false true)))
  (if (== _returnvalue_glider (Variables|Default|GetGlider%20ID))
    (Components|Activation|Activate (CallFunction|GetIdleVFXComponent) true)))

(fn ConstructionScript
  (|CallParentFunction))
