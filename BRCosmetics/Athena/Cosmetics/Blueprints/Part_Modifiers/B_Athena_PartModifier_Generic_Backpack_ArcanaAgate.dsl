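; Backpack cosmetic part modifier for Arcana Agate that changes the pack's picture when the player opens chests.
; On post-res-in it watches the OpenChests stat and triggers reactivity; in the front end it also plays the AnimBP chest-open reaction after a delay.
; Reactivity picks a random PictureID, checked against the previous one, and sends it to the mesh materials and, with a refresh counter, the idle VFX, adding sound outside the front end.
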(event EventOnPostResIn
  (bind _returnvalue (CustomCharacter|GetAssociatedPlayerPawn))
  (|CallParentFunction)
  (if (Utilities|Enum|Equal(Enum) (CustomCharacter|GetCharacterPartType) "Backpack")
    (bind _asskeletal_mesh_component (Utilities|Casting|CastToSkeletalMeshComponent (CustomCharacter|GetSkeletalMeshForPartOnPlayerPawn)))
    (bind _asf_med_arcana_agate_pack_anim_bp (Utilities|Casting|CastToF_MED_ArcanaAgate_Pack_AnimBP (Components|SkeletalMesh|GetAnimInstance _asskeletal_mesh_component)))
    (Variables|Default|SetAs%20F%20MED%20Arcana%20Agate%20Pack%20Anim%20BP _asf_med_arcana_agate_pack_anim_bp)
    ;[1 0.5998 0.226] Trigger reactivity on Chest Open
    (if (CustomCosmeticModifier|IsinFrontEnd)
      (Utilities|FlowControl|Delay (Variables|Default|GetDuration_Delay_FE)) ; Delay before activation
      (Class|FMEDArcanaAgatePackAnimBP|Chestopen (Variables|Default|GetAs%20F%20MED%20Arcana%20Agate%20Pack%20Anim%20BP))
      (CallFunction|TriggerReactivity)
      (Utilities|FlowControl|Delay (Variables|Default|GetDuration_Delay_FE)) ; Another delay to prevent FE reactivity from firing too frantically
    )
    (if (Utilities|IsValid _returnvalue :!class /Script/Engine.KismetSystemLibrary)
      (bind (_statname _statvalue) (Cosmetics|WatchCosmeticStat _returnvalue "(TagName=\"StatManager.OpenChests\")" 0 false)
        (:OnStatChanged
          (if (> _statvalue 0)
            (CallFunction|TriggerReactivity)))))))

(event Custom|Trigger_Reactivity
  (bind _returnvalue (CallFunction|GetIdleVFXComponent))
  ;[1 1 1] Random Picture for Material
  (if (Utilities|Enum|Equal(Enum) (CustomCharacter|GetCharacterPartType) "Backpack")
    (Variables|Default|SetRefreshInt (+ (Variables|Default|GetRefreshInt) 1))
    (Variables|Default|SetCurrentID (Math|Random|RandomIntegerinRange 0 2))
    (if (== (Variables|Default|GetPreviousID) (Variables|Default|GetCurrentID))
      _)
    (Rendering|Material|SetScalarParameterValueonMaterials (CustomCharacter|GetSkeletalMeshForPartOnPlayerPawn) "PictureID" (Math|Conversions|ToFloat(Integer) (Variables|Default|GetCurrentID)))
    (Variables|Default|SetPreviousID (Variables|Default|GetCurrentID))
    (Effects|Components|ParticleSystem|SetIntParameter _returnvalue "PictureID" (Variables|Default|GetCurrentID))
    (Effects|Components|ParticleSystem|SetIntParameter _returnvalue "RefreshInt" (Variables|Default|GetRefreshInt))
    (if (not (CustomCosmeticModifier|IsinFrontEnd))
      (Audio|SpawnSoundAttached (Variables|Default|GetOn%20Kill%20SOund) (CustomCharacter|GetSkeletalMeshForPartOnPlayerPawn)))))

(fn ConstructionScript
  (|CallParentFunction))
