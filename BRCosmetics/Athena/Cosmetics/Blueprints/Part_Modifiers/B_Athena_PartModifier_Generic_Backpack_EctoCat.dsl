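; Cosmetic part modifier for the EctoCat backpack that drives its idle particle effect from the pack's animation state.
; After res-in, it gets the backpack skeletal mesh's anim instance on the player pawn, casts it to the EctoCat pack AnimBP and caches it.
; It then copies that AnimBP's NotSprinting flag into the idle VFX "isSprinting" parameter, repeating every tick.
; The value passes through uninverted despite the opposite names.
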(event EventOnPostResIn
  ;[0.1 0.1 0.1] Init
  (|CallParentFunction)
  (bind _asm_med_ecto_cat_pack_anim_bp (Utilities|Casting|CastToM_MED_EctoCat_Pack_AnimBP (Components|SkeletalMesh|GetAnimInstance (Customization|GetSkeletalMeshforPartType (CustomCharacter|GetAssociatedPlayerPawn) (CustomCharacter|GetCharacterPartType) :!class /Script/FortniteGame.FortPlayerPawn))))
  (bind _output_get (Variables|Default|SetAs%20M%20MED%20Ecto%20Cat%20Pack%20Anim%20BP _asm_med_ecto_cat_pack_anim_bp))
  (Effects|Components|ParticleSystem|SetBoolParameter (CallFunction|GetIdleVFXComponent) "isSprinting" (Variables|/BRCosmetics/Accessories/FORT_Backpacks/Backpack_M_MED_EctoCat/Meshes/M_MED_EctoCat_Pack_AnimBP.M_MED_EctoCat_Pack_AnimBP_C|GetNotSprinting _output_get)))

(event EventTick
  (params (DeltaSeconds :type Float%20%28single-precision%29))
  ;[0.1 0.1 0.1] Tick
  (Effects|Components|ParticleSystem|SetBoolParameter (CallFunction|GetIdleVFXComponent) "isSprinting" (Variables|/BRCosmetics/Accessories/FORT_Backpacks/Backpack_M_MED_EctoCat/Meshes/M_MED_EctoCat_Pack_AnimBP.M_MED_EctoCat_Pack_AnimBP_C|GetNotSprinting (Variables|Default|GetAs%20M%20MED%20Ecto%20Cat%20Pack%20Anim%20BP))))

(fn ConstructionScript
  (|CallParentFunction))
