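; Goth Devil back bling part modifier that makes the backpack's idle particle effect react to damage.
; When cosmetic events can run, it watches the pawn's DamageReceived stat and, on a positive value, activates the effect and raises its FX_Reactivity parameter behind a DoOnce gate.
; In frontend display, a looping timer fires the effect periodically to preview the reaction, and is cleared on destruction.
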(event EventOnPostResIn
  (|CallParentFunction)
  (Networking|CanExecuteCosmeticEvents
    (:True
      (Utilities|FlowControl|Sequence
        (:then_0
          (bind (_statname _statvalue) (Cosmetics|WatchCosmeticStat (CustomCharacter|GetAssociatedPlayerPawn) "(TagName=\"StatManager.DamageReceived\")" -1)
            (:OnStatChanged
              (if (> _statvalue 0)
                (CallFunction|Reactivity)))))
        (:then_1
          ; FE
          (if (Reactive|Frontend|IsCosmeticObjectDisplayedinFrontendContext 0)
            (bind _returnvalue (Utilities|Time|SetTimerbyEvent (delegate Custom|FE) 5.0 true))
            (Variables|Default|SetFXTimer _returnvalue)))))))

(event Custom|FE
  (bind _returnvalue (CallFunction|GetIdleVFXComponent))
  ; FE
  (Effects|Components|ParticleSystem|SetFloatParameter _returnvalue "FX_Reactivity" 1.0)
  (Components|Activation|Activate _returnvalue false))

(event Custom|Reactivity
  (bind _returnvalue (CallFunction|GetIdleVFXComponent))
  (Utilities|FlowControl|DoOnce false
    (:Completed
      (Components|Activation|Activate _returnvalue true)
      (Effects|Components|ParticleSystem|SetFloatParameter _returnvalue "FX_Reactivity" 1.0)
      (Utilities|FlowControl|Delay 0.5))))

(event EventDestroyed
  (|CallParentFunction)
  (Utilities|Time|ClearandInvalidateTimerbyHandle (Variables|Default|GetFXTimer)))

(fn ConstructionScript
  (|CallParentFunction))
