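; Cosmetic part modifier for the WarmShadeWeasel backpack that drives its material animation from gameplay events through scalar parameters on the part's skeletal mesh.
; Skydive start and end toggle a skydiving state, and landing records match time and sets a game start time.
; Tick feeds a looping countdown, wrapping every configured number of minutes since skydiving ended, into TimeNumber.
; Stat changes push the new value as Kill_Number with an elimination timestamp, while spawn setup differs in the front end.
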(event EventOnBeginSkydiving
  (bind _mesh (Variables|Default|GetMesh))
  (Rendering|Material|SetScalarParameterValueonMaterials _mesh "SkyDivingTime" (Utilities|Time|GetGameTimeinSeconds))
  (Rendering|Material|SetScalarParameterValueonMaterials _mesh "SkyDivingActivate" 1.0))

(event EventOnEndSkydiving
  (bind _mesh (Variables|Default|GetMesh))
  (Rendering|Material|SetScalarParameterValueonMaterials _mesh "SkyDivingActivate" 0.0)
  (Variables|Default|SetEndSkyDivingTime (Fortnite|GetMatchTime))
  (Rendering|Material|SetScalarParameterValueonMaterials _mesh "GameStartTime" (- (Utilities|Time|GetGameTimeinSeconds) 2.0)))

(event EventTick
  (params (DeltaSeconds :type Float%20%28single-precision%29))
  (bind _output_get (Variables|Default|SetRemainingTime (- (Math|Conversions|ToFloat(Integer) (* (Variables|Default|GetMinutesLoop) 60)) (Math|Float|Division(WholeandRemainder) (- (Fortnite|GetMatchTime) (Variables|Default|GetEndSkyDivingTime)) (Math|Conversions|ToFloat(Integer) (* (Variables|Default|GetMinutesLoop) 60))))))
  (Rendering|Material|SetScalarParameterValueonMaterials (Variables|Default|GetMesh) "TimeNumber" _output_get))

(event EventOnPostResIn
  (bind _mesh (Variables|Default|GetMesh))
  (|CallParentFunction)
  (Variables|Default|SetMesh (CustomCharacter|GetSkeletalMeshForPartOnPlayerPawn))
  (if (CustomCosmeticModifier|IsinFrontEnd)
    (Rendering|Material|SetScalarParameterValueonMaterials _mesh "ElapsedLoopTime" 25.0)
    (else
      (Rendering|Material|SetScalarParameterValueonMaterials _mesh "ElimationTime" -4e+01))))

(event EventOnStatChanged
  (params (State%20Name :type Name :default "None") (Stat%20Value :type Integer))
  (bind _mesh (Variables|Default|GetMesh))
  (|CallParentFunction State%20Name Stat%20Value)
  (Rendering|Material|SetScalarParameterValueonMaterials _mesh "ElimationTime" (Utilities|Time|GetGameTimeinSeconds))
  (Rendering|Material|SetScalarParameterValueonMaterials _mesh "Kill_Number" (Math|Conversions|ToFloat(Integer) Stat%20Value)))

(event EventOnNewWeapons
  (params (WeaponsAsItems :type Array%20of%20Fort%20Item%20Object%20References)))

(event EventOnCurieStateAttached
  (params (Tag :type Gameplay%20Tag%20Structure%20%28by%20ref%29)))

(event EventOnCurieStateDetached
  (params (Tag :type Gameplay%20Tag%20Structure%20%28by%20ref%29)))

(event EventOnEnterVehicle
  (params (VehicleEntered :type Fort%20Vehicle%20Interface%20Interface%20%28by%20ref%29))
  (|CallParentFunction VehicleEntered) ; This node is disabled and will not be called.\nDrag off pins to build functionality.
)

(event EventOnExitVehicle
  (|CallParentFunction) ; This node is disabled and will not be called.\nDrag off pins to build functionality.
)

(event EventOnTargetingOrFirstPersonCameraChanged)

(event EventOnHitPawn
  (params (SelfActor :type Actor%20Object%20Reference) (OtherActor :type Actor%20Object%20Reference) (NormalImpulse :type Vector :default "0, 0, 0") (Hit :type Hit%20Result%20Structure%20%28by%20ref%29)))

(fn ConstructionScript
  (|CallParentFunction))
