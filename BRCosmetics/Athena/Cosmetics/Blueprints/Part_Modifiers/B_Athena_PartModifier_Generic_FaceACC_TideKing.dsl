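; Cosmetic part modifier for the Tide King face accessory that adjusts rendering and front-end slurp appearance.
; After the parent post-resource-init logic, it sets the part mesh's translucent sort priority and enables custom depth.
; In the front end only, it switches on the Progressive variant index with empty cases.
; It also scales down slurp inner color and chest mask radius on the Body mesh's dynamic material.
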(event EventOnPostResIn
  (bind _returnvalue (CustomCharacter|GetSkeletalMeshForPartOnPlayerPawn))
  (bind _returnvalue_2 (CustomCharacter|GetAssociatedPlayerPawn))
  (bind _returnvalue_3 (Customization|GetSkeletalMeshforPartType _returnvalue_2 "Body" :!class /Script/FortniteGame.FortPlayerPawn))
  (bind _material_instance_dynamic (Class|BTechArtGlobalLibrary|GetDynamicMaterialInstance _returnvalue_3 1))
  (|CallParentFunction)
  (Utilities|FlowControl|Sequence
    (:then_0
      (Rendering|SetTranslucentSortPriority _returnvalue 0)
      (Rendering|SetRenderCustomDepth _returnvalue true))
    (:then_1
      ;[1 1 1] Adjust slurp param in FE
      (if (CustomCosmeticModifier|IsinFrontEnd)
        (bind _asplayer_pawn_athena (Utilities|Casting|CastToPlayerPawn_Athena (CustomCharacter|GetAssociatedPlayerPawn)))
        (bind _index (CosmeticLoadout|GetVariantChannelActiveIndexforSlotActor "(TagName=\"LoadoutSlot.Character\")" _asplayer_pawn_athena "(TagName=\"Cosmetics.Variant.Channel.Progressive\")" (Utilities|Array|MakeArray "(TagName=\"Cosmetics.Variant.Property.Stage1\")" "(TagName=\"Cosmetics.Variant.Property.Stage2\")" "(TagName=\"Cosmetics.Variant.Property.Stage3\")") false false))
        (switch int _index
          (:0
            _)
          (:1
            _))
        (bind _returnvalue_4 (Rendering|Material|GetVectorParameterValue _material_instance_dynamic "Slurp_InnerColor" :!class /Script/Engine.MaterialInstanceDynamic))
        (Rendering|Material|SetVectorParameterValue _material_instance_dynamic "Slurp_InnerColor" (* _returnvalue_4 "(R=0.825000,G=0.825000,B=0.825000,A=1.000000)") :!class /Script/Engine.MaterialInstanceDynamic)
        (bind _returnvalue_5 (Rendering|Material|GetScalarParameterValue _material_instance_dynamic "TK_ChestMaskRadius" :!class /Script/Engine.MaterialInstanceDynamic))
        (Rendering|Material|SetScalarParameterValue _material_instance_dynamic "TK_ChestMaskRadius" (* _returnvalue_5 0.6) :!class /Script/Engine.MaterialInstanceDynamic)))))

(fn ConstructionScript
  (|CallParentFunction))
