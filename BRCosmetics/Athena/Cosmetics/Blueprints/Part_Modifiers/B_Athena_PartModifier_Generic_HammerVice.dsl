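; Cosmetic part modifier for the HammerVice pack that makes its idle VFX react when the player opens chests.
; On post res-in, it watches the OpenChests stat, or checks for frontend display, to trigger reactivity.
; Triggering raises the particle system's Reactivity flag, then clears it after a retriggerable delay.
; The pack's AnimBP chest-opening dispatcher feeds an integer ReactivityState into the same idle VFX.
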(event EventOnDayPhaseChanged
  (params (CurrentDayPhase :type EFortDayPhase%20Enum :default "Morning") (PreviousDayPhase :type EFortDayPhase%20Enum :default "Morning") (bAtCreation :type Boolean)))

(event EventOnNewPlayerWeaponType
  (params (NewType :type EFortItemType%20Enum :default "WorldItem") (OldType :type EFortItemType%20Enum :default "WorldItem"))
  (|CallParentFunction NewType OldType) ; This node is disabled and will not be called.\nDrag off pins to build functionality.
)

(event EventOnWeaponChanged
  (params (NewWeapon :type Fort%20Weapon%20Object%20Reference) (OldWeapon :type Fort%20Weapon%20Object%20Reference)))

(event EventOnNewWeapons
  (params (WeaponsAsItems :type Array%20of%20Fort%20Item%20Object%20References)))

(event EventOnCurieStateAttached
  (params (Tag :type Gameplay%20Tag%20Structure%20%28by%20ref%29)))

(event EventOnCurieStateDetached
  (params (Tag :type Gameplay%20Tag%20Structure%20%28by%20ref%29)))

(event EventOnBeginSkydiving
  (|CallParentFunction) ; This node is disabled and will not be called.\nDrag off pins to build functionality.
)

(event EventOnEndSkydiving
  (|CallParentFunction) ; This node is disabled and will not be called.\nDrag off pins to build functionality.
)

(event EventOnEnterVehicle
  (params (VehicleEntered :type Fort%20Vehicle%20Interface%20Interface%20%28by%20ref%29))
  (|CallParentFunction VehicleEntered) ; This node is disabled and will not be called.\nDrag off pins to build functionality.
)

(event EventOnExitVehicle
  (|CallParentFunction) ; This node is disabled and will not be called.\nDrag off pins to build functionality.
)

(event EventOnTargetingOrFirstPersonCameraChanged)

(event EventOnHitPawn
  (params (SelfActor :type Actor%20Object%20Reference) (OtherActor :type Actor%20Object%20Reference) (NormalImpulse :type Vector :default "0, 0, 0") (Hit :type Hit%20Result%20Structure%20%28by%20ref%29)))

(event EventOnPostResIn
  (bind _returnvalue (CustomCharacter|GetAssociatedPlayerPawn))
  ;[0.8333 0.06076 0.06076 0.5] CHEST OPEN REACTIVITY
  (|CallParentFunction)
  (bind _asskeletal_mesh_component (Utilities|Casting|CastToSkeletalMeshComponent (CustomCharacter|GetSkeletalMeshForPartOnPlayerPawn)))
  (bind _asf_med_hammer_vice_pack_anim_bp (Utilities|Casting|CastToF_MED_HammerVice_Pack_AnimBP (Components|SkeletalMesh|GetAnimInstance _asskeletal_mesh_component)))
  (Default|BindEventtoChestOpeningState _asf_med_hammer_vice_pack_anim_bp (EventDispatchers|CreateEvent :!event Custom|ReactivityState))
  (if (CallFunction|ShouldDisplayFrontendReactivity)
    ;[0.3183 0.8333 0.06076 0.5] FRONTEND
    (if (== (Variables|Default|GetVisual%20States) 0)
      (Variables|Default|SetVisual%20States (+ (Variables|Default|GetVisual%20States) 1))
      (Utilities|FlowControl|Delay 5.0)
      (if (== (Variables|Default|GetVisual%20States) 5)
        (Variables|Default|SetVisual%20States 0)))
    (CallFunction|TriggerReactivity))
  (if (Utilities|IsValid _returnvalue :!class /Script/Engine.KismetSystemLibrary)
    (bind (_statname _statvalue) (Cosmetics|WatchCosmeticStat _returnvalue "(TagName=\"StatManager.OpenChests\")" 0 false)
      (:OnStatChanged
        (if (> _statvalue 0)
          (CallFunction|TriggerReactivity))))))

(event Custom|Trigger_Reactivity
  (bind _returnvalue (CallFunction|GetIdleVFXComponent))
  ;[0.1895 0.06076 0.8333 0.5] REACTIVITY FX
  (if (Utilities|IsValid _returnvalue :!class /Script/Engine.KismetSystemLibrary)
    (Effects|Components|ParticleSystem|SetBoolParameter _returnvalue "Reactivity" true)
    (Utilities|FlowControl|RetriggerableDelay (Variables|Default|GetReactivity_Duration))
    (Effects|Components|ParticleSystem|SetBoolParameter _returnvalue "Reactivity" false)))

(event Custom|ReactivityState
  (params (ReactivityState :type Integer))
  ; ANIM BP REACTIVITY STATE
  (Effects|Components|ParticleSystem|SetIntParameter (CallFunction|GetIdleVFXComponent) "ReactivityState" ReactivityState))

(fn ConstructionScript
  (|CallParentFunction))
