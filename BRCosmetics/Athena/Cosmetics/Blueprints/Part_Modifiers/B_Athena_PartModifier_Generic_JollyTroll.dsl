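; Cosmetic part modifier for the JollyTroll (JollyTime) backpack that reacts to the wearer's open-chests stat.
; On post res-in, it calls the parent, then watches that stat on the associated player pawn.
; Each change calls PlayLibrarySound and prints the stat name and value for debugging.
; A separate OnStatChanged override calls the parent and prints the same; both greater-than-zero checks lead to no action.
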(event EventOnPostResIn
  (bind _watchopenchestsstattag (Variables|PartModifierControls|WatchCosmeticStat|GetWatchOpenChestsStatTag))
  (|CallParentFunction)
  (bind (_statname _statvalue) (Cosmetics|WatchCosmeticStat (CustomCharacter|GetAssociatedPlayerPawn) _watchopenchestsstattag 0 false)
    (:OnStatChanged
      (CallFunction|PlayLibrarySound _statname _statvalue _watchopenchestsstattag)
      (Development|PrintString (Utilities|String|Append "JollyTime Backpack- " "
" "Stat Name - " (Utilities|String|ToString(Name) _statname) "
" "Stat Value - " (Utilities|String|ToString(Integer) _statvalue)))
      (if (> _statvalue 0)
        _))))

(event EventOnStatChanged
  (params (State%20Name :type Name :default "None") (Stat%20Value :type Integer))
  (|CallParentFunction State%20Name Stat%20Value)
  (Development|PrintString (Utilities|String|Append "JollyTime Backpack- " "
" "Stat Name - " (Utilities|String|ToString(Name) State%20Name) "
" "Stat Value - " (Utilities|String|ToString(Integer) Stat%20Value)))
  (if (> Stat%20Value 0)
    _))

(fn ConstructionScript
  (|CallParentFunction))
