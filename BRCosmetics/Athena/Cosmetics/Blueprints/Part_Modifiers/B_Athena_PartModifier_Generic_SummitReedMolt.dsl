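; Cosmetic part modifier for SummitReedMolt that makes its idle VFX react to the player's kills.
; After res-in, a watched kill stat increments a local counter when exceeded in matches, while a looping timer fakes increments in the front end.
; The counter drives the particle "ReactivityCounter" parameter.
; Another looping timer copies the part anim instance's current LOD into the particle system as "SkeletalMeshCurrentLOD".
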(event Custom|FrontEndReactivity
  ; Front End Loop
  (Variables|Default|SetKillCount (+ (Variables|Default|GetKillCount) 1))
  ; FX
  (Effects|Components|ParticleSystem|SetIntParameter (CallFunction|GetIdleVFXComponent) "ReactivityCounter" (Variables|Default|GetKillCount)))

(event EventOnPostResIn
  (|CallParentFunction)
  (Utilities|FlowControl|Sequence
    (:then_0
      (if (CustomCosmeticModifier|IsinFrontEnd)
        ; Front End Loop
        (Utilities|Time|SetTimerbyEvent (delegate Custom|FrontEndReactivity) 1e+01 true)
        (Utilities|FlowControl|Delay 2.0)
        (else
          (bind (_statname _statvalue) (Cosmetics|WatchCosmeticStat (CustomCharacter|GetAssociatedPlayerPawn) "(TagName=\"StatManager.AthenaKills\")" 0 false)
            (:OnStatChanged
              (if (and (> _statvalue 0) (> _statvalue (Variables|Default|GetKillCount)))
                _)))))
      (Variables|Default|SetKillCount (+ (Variables|Default|GetKillCount) 1))
      ; FX
      (Effects|Components|ParticleSystem|SetIntParameter (CallFunction|GetIdleVFXComponent) "ReactivityCounter" (Variables|Default|GetKillCount)))
    (:then_1
      (Utilities|Time|SetTimerbyEvent (EventDispatchers|CreateEvent :!event Custom|LODCheck) (Variables|Default|GetLODCheckInterval) true)
      (bind _ascustom_character_part_anim_instance (Utilities|Casting|CastToCustomCharacterPartAnimInstance (Components|SkeletalMesh|GetAnimInstance (Customization|GetSkeletalMeshforPartType (CustomCharacter|GetAssociatedPlayerPawn) (CustomCharacter|GetCharacterPartType) :!class /Script/FortniteGame.FortPlayerPawn))))
      (Variables|Default|SetAnimInstance _ascustom_character_part_anim_instance))))

(event Custom|LODCheck
  ;[1 1 1] LOD Check
  (Effects|Components|ParticleSystem|SetIntParameter (CallFunction|GetIdleVFXComponent) "SkeletalMeshCurrentLOD" (Variables|/Script/FortniteGame.CustomCharacterPartAnimInstance|GetCurrentLOD (Variables|Default|GetAnimInstance))))

(fn ConstructionScript
  (|CallParentFunction))
