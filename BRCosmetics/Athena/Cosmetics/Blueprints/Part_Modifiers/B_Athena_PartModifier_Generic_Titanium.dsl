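; Cosmetic character part modifier for the generic Titanium part that plays reactive VFX and sound on eliminations and chest opens.
; Stat changes cover eliminations, while PostResIn watches the pawn's OpenChests stat and also triggers the effect in the front end.
; Trigger_Reactivity activates the idle VFX with its Empowered parameter, clears it after a retriggerable delay, and attaches a kill sound to the part mesh outside the front end.
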(event EventOnDayPhaseChanged
  (params (CurrentDayPhase :type EFortDayPhase%20Enum :default "Morning") (PreviousDayPhase :type EFortDayPhase%20Enum :default "Morning") (bAtCreation :type Boolean)))

(event EventOnNewPlayerWeaponType
  (params (NewType :type EFortItemType%20Enum :default "WorldItem") (OldType :type EFortItemType%20Enum :default "WorldItem"))
  (|CallParentFunction NewType OldType) ; This node is disabled and will not be called.\nDrag off pins to build functionality.
)

(event EventOnWeaponChanged
  (params (NewWeapon :type Fort%20Weapon%20Object%20Reference) (OldWeapon :type Fort%20Weapon%20Object%20Reference)))

(event EventOnNewWeapons
  (params (WeaponsAsItems :type Array%20of%20Fort%20Item%20Object%20References)))

(event EventOnCurieStateAttached
  (params (Tag :type Gameplay%20Tag%20Structure%20%28by%20ref%29)))

(event EventOnCurieStateDetached
  (params (Tag :type Gameplay%20Tag%20Structure%20%28by%20ref%29)))

(event EventOnBeginSkydiving
  (|CallParentFunction) ; This node is disabled and will not be called.\nDrag off pins to build functionality.
)

(event EventOnEndSkydiving
  (|CallParentFunction) ; This node is disabled and will not be called.\nDrag off pins to build functionality.
)

(event EventOnEnterVehicle
  (params (VehicleEntered :type Fort%20Vehicle%20Interface%20Interface%20%28by%20ref%29))
  (|CallParentFunction VehicleEntered) ; This node is disabled and will not be called.\nDrag off pins to build functionality.
)

(event EventOnExitVehicle
  (|CallParentFunction) ; This node is disabled and will not be called.\nDrag off pins to build functionality.
)

(event EventOnTargetingOrFirstPersonCameraChanged)

(event EventOnHitPawn
  (params (SelfActor :type Actor%20Object%20Reference) (OtherActor :type Actor%20Object%20Reference) (NormalImpulse :type Vector :default "0, 0, 0") (Hit :type Hit%20Result%20Structure%20%28by%20ref%29)))

(event EventOnStatChanged
  (params (State%20Name :type Name :default "None") (Stat%20Value :type Integer))
  ;[1 0.2361 0.3058] Trigger reactivity on Elimination
  (|CallParentFunction State%20Name Stat%20Value)
  (if (> Stat%20Value 0)
    (CallFunction|TriggerReactivity)))

(event EventOnPostResIn
  (bind _returnvalue (CustomCharacter|GetAssociatedPlayerPawn))
  ;[1 0.5998 0.226] Trigger reactivity on Chest Open
  (|CallParentFunction)
  (if (CustomCosmeticModifier|IsinFrontEnd)
    (CallFunction|TriggerReactivity)
    (Utilities|FlowControl|Delay (Variables|Default|GetDuration_Delay_FE)))
  (if (Utilities|IsValid _returnvalue :!class /Script/Engine.KismetSystemLibrary)
    (bind (_statname _statvalue) (Cosmetics|WatchCosmeticStat _returnvalue "(TagName=\"StatManager.OpenChests\")" 0 false)
      (:OnStatChanged
        (if (> _statvalue 0)
          (CallFunction|TriggerReactivity))))))

(event Custom|Trigger_Reactivity
  (bind _returnvalue (CallFunction|GetIdleVFXComponent))
  ;[0.4779 1 0.236] Reactivity FX
  (if (Utilities|IsValid _returnvalue :!class /Script/Engine.KismetSystemLibrary)
    (Effects|Components|ParticleSystem|SetBoolParameter _returnvalue "Empowered" true)
    (Components|Activation|Activate _returnvalue true)
    (Utilities|FlowControl|Sequence
      (:then_0
        (Utilities|FlowControl|RetriggerableDelay (Variables|Default|GetDuration_Empower))
        (Effects|Components|ParticleSystem|SetBoolParameter _returnvalue "Empowered" false))
      (:then_1
        (if (not (CustomCosmeticModifier|IsinFrontEnd))
          (Audio|SpawnSoundAttached (Variables|Default|GetOnKillSound) (CustomCharacter|GetSkeletalMeshForPartOnPlayerPawn)))))))

(fn PlaySound
  (params (NewParam :type Boolean)))

(fn ConstructionScript
  (|CallParentFunction))
