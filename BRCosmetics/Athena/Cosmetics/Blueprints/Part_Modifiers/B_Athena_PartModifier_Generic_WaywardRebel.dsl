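; Cosmetic part modifier for the Wayward Rebel character that attaches Niagara effects to both sides of the head and to a left purse bone.
; In the front end, a looping timer started after res-in respawns the effects and is cleared on destroy.
; A positive stat change spawns them through a DoOnce if the head is visible, also enlarging the purse effect.
; Niagara bool parameters mark the left side and hide the purse effect's light.
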(event Custom|CE_Loop_Frontend
  (bind _returnvalue (CustomCharacter|GetSkeletalMeshForPartOnPlayerPawn))
  (bind _returnvalue_2 (PartModifier|SpawnSystem|SpawnSystemAttachedPartModifier (Variables|Default|GetNiagaraSystem) _returnvalue "head" "2.500000,0.500000,12.000000" "90.000000,0.000000,0.000000")
    (:Success
      (Effects|Components|ParticleSystem|SetBoolParameter _returnvalue_2 "IsLeft" true)
      (PartModifier|SpawnSystem|SpawnSystemAttachedPartModifier (Variables|Default|GetNiagaraSystem) _returnvalue "head" "3.000000,0.500000,-12.000000" "-90.000000,0.000000,0.000000"
        (:Success
          (bind _returnvalue_3 (PartModifier|SpawnSystem|SpawnSystemAttachedPartModifier (Variables|Default|GetNiagaraSystem) _returnvalue "dyn_purse_03_l" "2.000000,0.000000,4.000000" "90.000000,0.000000,0.000000")
            (:Success
              (Effects|Components|ParticleSystem|SetBoolParameter _returnvalue_3 "Hide light" true))))))))

(event EventOnPostResIn
  (|CallParentFunction)
  (if (CustomCosmeticModifier|IsinFrontEnd)
    (bind _returnvalue (Utilities|Time|SetTimerbyEvent (delegate Custom|CE_Loop_Frontend) 3.0 true))
    (Variables|Default|SetTimer _returnvalue)))

(event EventOnStatChanged
  (params (State%20Name :type Name :default "None") (Stat%20Value :type Integer))
  (bind _returnvalue (CustomCharacter|GetSkeletalMeshForPartOnPlayerPawn))
  (|CallParentFunction State%20Name Stat%20Value)
  (if (> Stat%20Value 0)
    (if (not (Variables|HiddenPartCheck|GetbIsHeadHidden))
      (Utilities|FlowControl|DoOnce false
        (:Completed
          (bind _returnvalue_2 (PartModifier|SpawnSystem|SpawnSystemAttachedPartModifier (Variables|Default|GetNiagaraSystem) _returnvalue "head" "2.500000,0.500000,12.000000" "90.000000,0.000000,0.000000")
            (:Success
              (Effects|Components|ParticleSystem|SetBoolParameter _returnvalue_2 "IsLeft" true)
              (PartModifier|SpawnSystem|SpawnSystemAttachedPartModifier (Variables|Default|GetNiagaraSystem) _returnvalue "head" "3.000000,0.500000,-12.000000" "-90.000000,0.000000,0.000000"
                (:Success
                  (bind _returnvalue_3 (PartModifier|SpawnSystem|SpawnSystemAttachedPartModifier (Variables|Default|GetNiagaraSystem) _returnvalue "dyn_purse_03_l" "2.000000,0.000000,4.000000" "90.000000,0.000000,0.000000")
                    (:Success
                      (Transformation|SetRelativeScale3D _returnvalue_3 "1.300000,1.300000,1.300000")
                      (Effects|Components|ParticleSystem|SetBoolParameter _returnvalue_3 "Hide light" true)
                      (Utilities|FlowControl|RetriggerableDelay 1.3))))))))))))

(event EventDestroyed
  (|CallParentFunction)
  (if (CustomCosmeticModifier|IsinFrontEnd)
    (Utilities|Time|ClearandInvalidateTimerbyHandle (Variables|Default|GetTimer))))

(fn ConstructionScript
  (|CallParentFunction))
