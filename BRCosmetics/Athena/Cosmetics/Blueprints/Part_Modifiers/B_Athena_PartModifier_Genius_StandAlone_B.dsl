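; Character part modifier for the Genius standalone B cosmetic, driving reactive neon face effects on the pawn's part mesh.
; After post-res-in and a toon frontend LUT swap, it watches damage, elimination and chest-opening stats.
; Each change timestamps the neon materials, sets reaction length and mouth shape, and raises overlay VertexHide until a retriggerable delay clears it.
; Initialization also suppresses the spawn reaction and adjusts audio-reactive parameters based on a frontend check.
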(event EventOnPostResIn
  (bind _returnvalue (CustomCharacter|GetSkeletalMeshForPartOnPlayerPawn))
  (bind _returnvalue_2 (CustomCharacter|GetSkeletalMeshForPartOnPlayerPawn))
  (bind _returnvalue_3 (CustomCharacter|GetSkeletalMeshForPartOnPlayerPawn))
  (bind _returnvalue_4 (CustomCharacter|GetSkeletalMeshForPartOnPlayerPawn))
  (|CallParentFunction)
  (Utilities|ToonFrontendLUTReplace
    (:Out
      (Utilities|FlowControl|Sequence
        (:then_1
          ;[1 1 1] Damage Received
          (Cosmetics|WatchCosmeticStat (CustomCharacter|GetAssociatedPlayerPawn) "(TagName=\"StatManager.DamageReceived\")" 0 false
            (:OnStatChanged
              (Rendering|Material|SetScalarParameterValueonMaterials _returnvalue_3 "Neon_GameTimeFromBP" (Utilities|Time|GetGameTimeinSeconds))
              (Rendering|Material|SetScalarParameterValueonMaterials _returnvalue_3 "Neon_ReactiveLength" 0.8)
              (Rendering|Material|SetScalarParameterValueonMaterials _returnvalue_3 "Neon_ReactiveShape_Mouth" 1.0)
              (bind _returnvalue_5 (Class|FortCosmeticMaterialBlueprintLibrary|GetorCreateDynamicOverlayMaterial _returnvalue_3))
              (Rendering|Material|SetScalarParameterValue _returnvalue_5 "VertexHide Threshold" 1.0 :!class /Script/Engine.MaterialInstanceDynamic)
              (Utilities|FlowControl|RetriggerableDelay 1.2)
              (Rendering|Material|SetScalarParameterValue _returnvalue_5 "VertexHide Threshold" 0.0 :!class /Script/Engine.MaterialInstanceDynamic))))
        (:then_2
          ;[1 1 1] Elimination
          (Cosmetics|WatchCosmeticStat (CustomCharacter|GetAssociatedPlayerPawn) "(TagName=\"StatManager.AthenaKills\")" 0 false
            (:OnStatChanged
              (Utilities|FlowControl|Delay 0.01)
              (Rendering|Material|SetScalarParameterValueonMaterials _returnvalue_4 "Neon_GameTimeFromBP" (Utilities|Time|GetGameTimeinSeconds))
              (Rendering|Material|SetScalarParameterValueonMaterials _returnvalue_4 "Neon_ReactiveLength" 3.0)
              (Rendering|Material|SetScalarParameterValueonMaterials _returnvalue_4 "Neon_ReactiveShape_Mouth" 0.5)
              (bind _returnvalue_6 (Class|FortCosmeticMaterialBlueprintLibrary|GetorCreateDynamicOverlayMaterial _returnvalue_4))
              (Rendering|Material|SetScalarParameterValue _returnvalue_6 "VertexHide Threshold" 0.5 :!class /Script/Engine.MaterialInstanceDynamic)
              (Utilities|FlowControl|RetriggerableDelay 4.5)
              (Rendering|Material|SetScalarParameterValue _returnvalue_6 "VertexHide Threshold" 0.0 :!class /Script/Engine.MaterialInstanceDynamic))))
        (:then_3
          ;[1 1 1] openChest
          (Cosmetics|WatchCosmeticStat (CustomCharacter|GetAssociatedPlayerPawn) "(TagName=\"StatManager.OpenChests\")" 0 false
            (:OnStatChanged
              (Rendering|Material|SetScalarParameterValueonMaterials _returnvalue_2 "Neon_GameTimeFromBP" (Utilities|Time|GetGameTimeinSeconds))
              (Rendering|Material|SetScalarParameterValueonMaterials _returnvalue_2 "Neon_ReactiveLength" 3.0)
              (Rendering|Material|SetScalarParameterValueonMaterials _returnvalue_2 "Neon_ReactiveShape_Mouth" 0.5)
              (bind _returnvalue_7 (Class|FortCosmeticMaterialBlueprintLibrary|GetorCreateDynamicOverlayMaterial _returnvalue_2))
              (Rendering|Material|SetScalarParameterValue _returnvalue_7 "VertexHide Threshold" 0.5 :!class /Script/Engine.MaterialInstanceDynamic)
              (Utilities|FlowControl|RetriggerableDelay 4.5)
              (Rendering|Material|SetScalarParameterValue _returnvalue_7 "VertexHide Threshold" 0.0 :!class /Script/Engine.MaterialInstanceDynamic))))
        (:then_4
          ;[1 1 1] Stop Reaction on Spawn + Reduce Audio Amplitude in FE
          (Rendering|Material|SetScalarParameterValueonMaterials _returnvalue "Neon_GameTimeFromBP" -1e+01)
          (if (CustomCosmeticModifier|IsinFrontEnd)
            (Rendering|Material|SetScalarParameterValueonMaterials _returnvalue "Neon_EventAudioMultiplier" 0.8)
            (Rendering|Material|SetScalarParameterValueonMaterials _returnvalue "Neon_AudioColorPulseStrength" 0.3)))))))

(fn GameTimeFromBPForGlow
  (params (A :type Float%20%28double-precision%29) (B :type Float%20%28double-precision%29))
  (return 0.0))

(fn ConstructionScript)
