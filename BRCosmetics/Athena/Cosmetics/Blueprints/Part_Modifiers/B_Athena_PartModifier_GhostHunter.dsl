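; Ghost Hunter backpack part modifier that shows player kills on a left counter and AI kills on a right counter.
; On init it creates a dark or light dynamic material from the variant, mapping clamped stats through the vertexColor array.
; Stat watchers update the material, spawn Niagara FX and play timelines for color, pip offset and overlay.
; In frontend preview, delayed FlipFlop recursion alternately fills both counters in a loop.
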(event Custom|LoopFrontEnd
  (bind _ref_backpackmat (Variables|Default|GetREF_BackpackMat))
  (bind _ref_backpackmat_2 (Variables|Default|GetREF_BackpackMat))
  (bind _ref_backpackmat_3 (Variables|Default|GetREF_BackpackMat))
  ;[0.266 1 0.6575 1 200] REFS
  ;;[1 1 1] Reset Params for FrontEnd Playback Loop
  (Variables|Default|SetCounter_Left 0.0)
  (Variables|Default|SetCounter_Right 0.0)
  (Rendering|Material|SetScalarParameterValue (Variables|Default|GetREF_BackpackMat) "Overlay_Lerp" 0.0 :!class /Script/Engine.MaterialInstanceDynamic)
  (bind _result (ArrayMacros|SafeArrayGet (Variables|Default|GetvertexColor) (Math|Integer|Clamp(Integer) (stale EventOnPostResIn _returnvalue_2) 0 10))
    (:%20%20
      (Rendering|Material|SetScalarParameterValue (stale EventOnPostResIn _output_get) "LeftCounterProgress" _result :!class /Script/Engine.MaterialInstanceDynamic)
      (bind _returnvalue (Class|FortCosmeticFunctionLibrary|GetCosmeticStatValue (CustomCharacter|GetAssociatedPlayerPawn) "(TagName=\"StatManager.AIKills\")"))
      (bind _result_2 (ArrayMacros|SafeArrayGet (Variables|Default|GetvertexColor) (Math|Integer|Clamp(Integer) _returnvalue 0 10))
        (:%20%20
          (Rendering|Material|SetScalarParameterValue (stale EventOnPostResIn _output_get) "RightCounter_Progress" _result_2 :!class /Script/Engine.MaterialInstanceDynamic)
          (Utilities|FlowControl|Sequence
            (:then_0
              (if (not (>= (stale EventOnPostResIn _returnvalue_2) 10))
                (bind _returnvalue_2 (Class|FortCosmeticFunctionLibrary|GetCosmeticStatValue (CustomCharacter|GetAssociatedPlayerPawn) "(TagName=\"StatManager.AIKills\")"))
                (if (>= _returnvalue_2 10)
                  _))
              (Rendering|Material|SetScalarParameterValue _ref_backpackmat_3 "Overlay_Lerp" 1.0 :!class /Script/Engine.MaterialInstanceDynamic))
            (:then_1
              (Rendering|Material|SetScalarParameterValue _ref_backpackmat_3 "Color_LeftSideMultiplier" 0.0 :!class /Script/Engine.MaterialInstanceDynamic)
              (Rendering|Material|SetScalarParameterValue _ref_backpackmat_3 "WPO_SelectedHeight" 0.0 :!class /Script/Engine.MaterialInstanceDynamic)
              (Rendering|Material|SetScalarParameterValue _ref_backpackmat_3 "Color_RightSideMultiplier" 0.0 :!class /Script/Engine.MaterialInstanceDynamic)
              (Rendering|Material|SetScalarParameterValue _ref_backpackmat_3 "WPO_SelectedHeight_RightSide" 0.0 :!class /Script/Engine.MaterialInstanceDynamic)
              ;[0.141 0.4403 1 1 200] FRONTEND
              (if (Reactive|Frontend|IsCosmeticObjectDisplayedasPreview (CustomCharacter|GetAssociatedPlayerPawn))
                (Utilities|FlowControl|RetriggerableDelay 1.0)
                (bind _returnvalue_3 (Niagara|SpawnSystemAttached "/BRCosmetics/Accessories/FORT_Backpacks/Backpack_F_MED_Ghost_Hunter/FX/NS_Backpack_GhostHunter.NS_Backpack_GhostHunter" (CustomCharacter|GetSkeletalMeshForPartOnPlayerPawn) "pack" "0, 0, 0" "0, 0, 0" "SnapToTarget" true true "AutoRelease"))
                (CallFunction|GhostHunterSetFXFrontEndFlag _returnvalue_3)
                (Utilities|FlowControl|FlipFlop
                  (:A
                    ;;[1 1 1 1 200] LEFTSIDE
                    (if (not (>= (Variables|Default|GetCounter_Left) 11.0))
                      (bind _output_get (Variables|Default|SetCounter_Left (+ (Variables|Default|GetCounter_Left) 1.0)))
                      (bind _result_3 (ArrayMacros|SafeArrayGet (Variables|Default|GetvertexColor) (Math|Float|Truncate _output_get))
                        (:%20%20
                          (Rendering|Material|SetScalarParameterValue (Variables|Default|GetREF_BackpackMat) "LeftCounterProgress" _result_3 :!class /Script/Engine.MaterialInstanceDynamic)
                          (if (>= (Variables|Default|GetCounter_Left) 11.0)
                            (Utilities|FlowControl|Sequence
                              (:then_0
                                ;;;[1 0.4278 0.2058] ON KILL BEFORE OVERLAY
                                (bind (_direction _colorintensity) (Timeline|Timeline_1 PlayFromStart)
                                  (:Update
                                    (Rendering|Material|SetScalarParameterValue _ref_backpackmat "Color_LeftSideMultiplier" _colorintensity :!class /Script/Engine.MaterialInstanceDynamic))))
                              (:then_1
                                (bind (_direction_2 _pipmove) (Timeline|Timeline_2 PlayFromStart)
                                  (:Update
                                    (Rendering|Material|SetScalarParameterValue _ref_backpackmat "WPO_SelectedHeight" _pipmove :!class /Script/Engine.MaterialInstanceDynamic))))
                              (:then_2
                                _)
                              (:then_3
                                ;;;[1 1 1] OVERLAY
                                (bind (_direction_3 _overlayintensity) (Timeline|Timeline_0 PlayFromStart)
                                  (:Update
                                    (Rendering|Material|SetScalarParameterValue (Variables|Default|GetREF_BackpackMat) "Overlay_Lerp" _overlayintensity :!class /Script/Engine.MaterialInstanceDynamic)))))
                            (else
                              (Utilities|FlowControl|Sequence
                                (:then_0
                                  (Timeline|Timeline_1 PlayFromStart))
                                (:then_1
                                  (Timeline|Timeline_2 PlayFromStart))
                                (:then_2
                                  _)
                                (:then_3
                                  (CallFunction|ReplayCountDown)))))
                          (bind _returnvalue_4 (PartModifier|SpawnSystem|SpawnSystemAttachedPartModifier "/BRCosmetics/Accessories/FORT_Backpacks/Backpack_F_MED_Ghost_Hunter/FX/NS_Backpack_GhostHunter.NS_Backpack_GhostHunter" (CustomCharacter|GetSkeletalMeshForPartOnPlayerPawn) "Root")
                            (:Success
                              (CallFunction|GhostHunterSetFXFrontEndFlag _returnvalue_4)))))))
                  (:B
                    ;;[1 0.575 0.2453 1 200] RIGHTSIDE
                    (if (not (>= (Variables|Default|GetCounter_Right) 11.0))
                      (bind _output_get_2 (Variables|Default|SetCounter_Right (+ (Variables|Default|GetCounter_Right) 1.0)))
                      (bind _result_4 (ArrayMacros|SafeArrayGet (Variables|Default|GetvertexColor) (Math|Float|Truncate _output_get_2))
                        (:%20%20
                          (Rendering|Material|SetScalarParameterValue (Variables|Default|GetREF_BackpackMat) "RightCounter_Progress" _result_4 :!class /Script/Engine.MaterialInstanceDynamic)
                          (if (>= (Variables|Default|GetCounter_Right) 1e+01)
                            (Utilities|FlowControl|Sequence
                              (:then_0
                                ;;;[1 0.4278 0.2058] ON KILL BEFORE OVERLAY
                                (bind (_direction_4 _colorintensity_2) (Timeline|Timeline_4 PlayFromStart)
                                  (:Update
                                    (Rendering|Material|SetScalarParameterValue _ref_backpackmat_2 "Color_RightSideMultiplier" _colorintensity_2 :!class /Script/Engine.MaterialInstanceDynamic))))
                              (:then_1
                                (bind (_direction_5 _pipmove_2) (Timeline|Timeline_5 PlayFromStart)
                                  (:Update
                                    (Rendering|Material|SetScalarParameterValue _ref_backpackmat_2 "WPO_SelectedHeight_RightSide" _pipmove_2 :!class /Script/Engine.MaterialInstanceDynamic))))
                              (:then_2
                                _)
                              (:then_3
                                ;;;[1 1 1] OVERLAY
                                (bind (_direction_6 _overlayintensity_2) (Timeline|Timeline_3 PlayFromStart)
                                  (:Update
                                    (Rendering|Material|SetScalarParameterValue (Variables|Default|GetREF_BackpackMat) "Overlay_Lerp" _overlayintensity_2 :!class /Script/Engine.MaterialInstanceDynamic))
                                  (:Finished
                                    (Utilities|FlowControl|Delay 3.0)
                                    (CallFunction|LoopFrontEnd)))))
                            (else
                              (Utilities|FlowControl|Sequence
                                (:then_0
                                  (Timeline|Timeline_4 PlayFromStart))
                                (:then_1
                                  (Timeline|Timeline_5 PlayFromStart))
                                (:then_2
                                  _)
                                (:then_3
                                  (CallFunction|ReplayCountDown)))))
                          (bind _returnvalue_5 (PartModifier|SpawnSystem|SpawnSystemAttachedPartModifier "/BRCosmetics/Accessories/FORT_Backpacks/Backpack_F_MED_Ghost_Hunter/FX/NS_Backpack_GhostHunter.NS_Backpack_GhostHunter" (CustomCharacter|GetSkeletalMeshForPartOnPlayerPawn) "Root")
                            (:Success
                              (CallFunction|GhostHunterSetFXFrontEndFlag _returnvalue_5))))))))))))))))

(event EventOnPostResIn
  (bind _ref_backpackmat (Variables|Default|GetREF_BackpackMat))
  (bind _ref_backpackmat_2 (Variables|Default|GetREF_BackpackMat))
  (bind _ref_backpackmat_3 (Variables|Default|GetREF_BackpackMat))
  ;[0.266 1 0.6575 1 200] REFS
  (|CallParentFunction)
  (Utilities|FlowControl|RetriggerableDelay 0.1)
  (bind _index (CosmeticLoadout|GetVariantChannelActiveIndexforSlotActor "(TagName=\"LoadoutSlot.Backpack\")" (CustomCharacter|GetAssociatedPlayerPawn) "(TagName=\"Cosmetics.Variant.Channel.Progressive\")" (Utilities|Array|MakeArray "(TagName=\"Cosmetics.Variant.Property.Stage1\")" "(TagName=\"Cosmetics.Variant.Property.Stage2\")")))
  (bind _returnvalue (Rendering|Material|CreateDynamicMaterialInstance (Utilities|Select :Index _index :Option%200 (Variables|Default|GetREF_Dark) :Option%201 (Variables|Default|GetREF_Light)) "newInstance" "None" :!class /Script/Engine.KismetMaterialLibrary))
  (bind _output_get (Variables|Default|SetREF_BackpackMat _returnvalue))
  (Rendering|Material|SetMaterial (CustomCharacter|GetSkeletalMeshForPartOnPlayerPawn) 1 _output_get)
  (bind _returnvalue_2 (Class|FortCosmeticFunctionLibrary|GetCosmeticStatValue (CustomCharacter|GetAssociatedPlayerPawn) "(TagName=\"StatManager.AthenaKills\")"))
  (bind _result (ArrayMacros|SafeArrayGet (Variables|Default|GetvertexColor) (Math|Integer|Clamp(Integer) _returnvalue_2 0 10))
    (:%20%20
      (Rendering|Material|SetScalarParameterValue _output_get "LeftCounterProgress" _result :!class /Script/Engine.MaterialInstanceDynamic)
      (bind _returnvalue_3 (Class|FortCosmeticFunctionLibrary|GetCosmeticStatValue (CustomCharacter|GetAssociatedPlayerPawn) "(TagName=\"StatManager.AIKills\")"))
      (bind _result_2 (ArrayMacros|SafeArrayGet (Variables|Default|GetvertexColor) (Math|Integer|Clamp(Integer) _returnvalue_3 0 10))
        (:%20%20
          (Rendering|Material|SetScalarParameterValue _output_get "RightCounter_Progress" _result_2 :!class /Script/Engine.MaterialInstanceDynamic)
          (Utilities|FlowControl|Sequence
            (:then_0
              (if (not (>= _returnvalue_2 10))
                (bind _returnvalue_4 (Class|FortCosmeticFunctionLibrary|GetCosmeticStatValue (CustomCharacter|GetAssociatedPlayerPawn) "(TagName=\"StatManager.AIKills\")"))
                (if (>= _returnvalue_4 10)
                  _))
              (Rendering|Material|SetScalarParameterValue _ref_backpackmat_3 "Overlay_Lerp" 1.0 :!class /Script/Engine.MaterialInstanceDynamic))
            (:then_1
              (Rendering|Material|SetScalarParameterValue _ref_backpackmat_3 "Color_LeftSideMultiplier" 0.0 :!class /Script/Engine.MaterialInstanceDynamic)
              (Rendering|Material|SetScalarParameterValue _ref_backpackmat_3 "WPO_SelectedHeight" 0.0 :!class /Script/Engine.MaterialInstanceDynamic)
              (Rendering|Material|SetScalarParameterValue _ref_backpackmat_3 "Color_RightSideMultiplier" 0.0 :!class /Script/Engine.MaterialInstanceDynamic)
              (Rendering|Material|SetScalarParameterValue _ref_backpackmat_3 "WPO_SelectedHeight_RightSide" 0.0 :!class /Script/Engine.MaterialInstanceDynamic)
              ;[0.141 0.4403 1 1 200] FRONTEND
              (if (Reactive|Frontend|IsCosmeticObjectDisplayedasPreview (CustomCharacter|GetAssociatedPlayerPawn))
                (Utilities|FlowControl|RetriggerableDelay 1.0)
                (bind _returnvalue_5 (Niagara|SpawnSystemAttached "/BRCosmetics/Accessories/FORT_Backpacks/Backpack_F_MED_Ghost_Hunter/FX/NS_Backpack_GhostHunter.NS_Backpack_GhostHunter" (CustomCharacter|GetSkeletalMeshForPartOnPlayerPawn) "pack" "0, 0, 0" "0, 0, 0" "SnapToTarget" true true "AutoRelease"))
                (CallFunction|GhostHunterSetFXFrontEndFlag _returnvalue_5)
                (Utilities|FlowControl|FlipFlop
                  (:A
                    ;;[1 1 1 1 200] LEFTSIDE
                    (if (not (>= (Variables|Default|GetCounter_Left) 11.0))
                      (bind _output_get_2 (Variables|Default|SetCounter_Left (+ (Variables|Default|GetCounter_Left) 1.0)))
                      (bind _result_3 (ArrayMacros|SafeArrayGet (Variables|Default|GetvertexColor) (Math|Float|Truncate _output_get_2))
                        (:%20%20
                          (Rendering|Material|SetScalarParameterValue (Variables|Default|GetREF_BackpackMat) "LeftCounterProgress" _result_3 :!class /Script/Engine.MaterialInstanceDynamic)
                          (if (>= (Variables|Default|GetCounter_Left) 11.0)
                            (Utilities|FlowControl|Sequence
                              (:then_0
                                ;;;[1 0.4278 0.2058] ON KILL BEFORE OVERLAY
                                (Timeline|Timeline_1 PlayFromStart))
                              (:then_1
                                (Timeline|Timeline_2 PlayFromStart))
                              (:then_2
                                _)
                              (:then_3
                                ;;;[1 1 1] OVERLAY
                                (Timeline|Timeline_0 PlayFromStart)))
                            (else
                              (Utilities|FlowControl|Sequence
                                (:then_0
                                  (Timeline|Timeline_1 PlayFromStart))
                                (:then_1
                                  (Timeline|Timeline_2 PlayFromStart))
                                (:then_2
                                  _)
                                (:then_3
                                  (CallFunction|ReplayCountDown)))))
                          (bind _returnvalue_6 (PartModifier|SpawnSystem|SpawnSystemAttachedPartModifier "/BRCosmetics/Accessories/FORT_Backpacks/Backpack_F_MED_Ghost_Hunter/FX/NS_Backpack_GhostHunter.NS_Backpack_GhostHunter" (CustomCharacter|GetSkeletalMeshForPartOnPlayerPawn) "Root")
                            (:Success
                              (CallFunction|GhostHunterSetFXFrontEndFlag _returnvalue_6)))))))
                  (:B
                    ;;[1 0.575 0.2453 1 200] RIGHTSIDE
                    (if (not (>= (Variables|Default|GetCounter_Right) 11.0))
                      (bind _output_get_3 (Variables|Default|SetCounter_Right (+ (Variables|Default|GetCounter_Right) 1.0)))
                      (bind _result_4 (ArrayMacros|SafeArrayGet (Variables|Default|GetvertexColor) (Math|Float|Truncate _output_get_3))
                        (:%20%20
                          (Rendering|Material|SetScalarParameterValue (Variables|Default|GetREF_BackpackMat) "RightCounter_Progress" _result_4 :!class /Script/Engine.MaterialInstanceDynamic)
                          (if (>= (Variables|Default|GetCounter_Right) 1e+01)
                            (Utilities|FlowControl|Sequence
                              (:then_0
                                ;;;[1 0.4278 0.2058] ON KILL BEFORE OVERLAY
                                (Timeline|Timeline_4 PlayFromStart))
                              (:then_1
                                (Timeline|Timeline_5 PlayFromStart))
                              (:then_2
                                _)
                              (:then_3
                                ;;;[1 1 1] OVERLAY
                                (Timeline|Timeline_3 PlayFromStart)))
                            (else
                              (Utilities|FlowControl|Sequence
                                (:then_0
                                  (Timeline|Timeline_4 PlayFromStart))
                                (:then_1
                                  (Timeline|Timeline_5 PlayFromStart))
                                (:then_2
                                  _)
                                (:then_3
                                  (CallFunction|ReplayCountDown)))))
                          (bind _returnvalue_7 (PartModifier|SpawnSystem|SpawnSystemAttachedPartModifier "/BRCosmetics/Accessories/FORT_Backpacks/Backpack_F_MED_Ghost_Hunter/FX/NS_Backpack_GhostHunter.NS_Backpack_GhostHunter" (CustomCharacter|GetSkeletalMeshForPartOnPlayerPawn) "Root")
                            (:Success
                              (CallFunction|GhostHunterSetFXFrontEndFlag _returnvalue_7))))))))))))))))

(event Custom|ReplayCountDown
  (bind _ref_backpackmat (Variables|Default|GetREF_BackpackMat))
  (bind _ref_backpackmat_2 (Variables|Default|GetREF_BackpackMat))
  ;[0.141 0.4403 1 1 200] FRONTEND
  (Utilities|FlowControl|RetriggerableDelay 1.0)
  (bind _returnvalue (Niagara|SpawnSystemAttached "/BRCosmetics/Accessories/FORT_Backpacks/Backpack_F_MED_Ghost_Hunter/FX/NS_Backpack_GhostHunter.NS_Backpack_GhostHunter" (CustomCharacter|GetSkeletalMeshForPartOnPlayerPawn) "pack" "0, 0, 0" "0, 0, 0" "SnapToTarget" true true "AutoRelease"))
  (CallFunction|GhostHunterSetFXFrontEndFlag _returnvalue)
  (Utilities|FlowControl|FlipFlop
    (:A
      ;;[1 1 1 1 200] LEFTSIDE
      (if (not (>= (Variables|Default|GetCounter_Left) 11.0))
        (bind _output_get (Variables|Default|SetCounter_Left (+ (Variables|Default|GetCounter_Left) 1.0)))
        (bind _result (ArrayMacros|SafeArrayGet (Variables|Default|GetvertexColor) (Math|Float|Truncate _output_get))
          (:%20%20
            (Rendering|Material|SetScalarParameterValue (Variables|Default|GetREF_BackpackMat) "LeftCounterProgress" _result :!class /Script/Engine.MaterialInstanceDynamic)
            (if (>= (Variables|Default|GetCounter_Left) 11.0)
              (Utilities|FlowControl|Sequence
                (:then_0
                  ;;;[1 0.4278 0.2058] ON KILL BEFORE OVERLAY
                  (Timeline|Timeline_1 PlayFromStart))
                (:then_1
                  (Timeline|Timeline_2 PlayFromStart))
                (:then_2
                  _)
                (:then_3
                  ;;;[1 1 1] OVERLAY
                  (Timeline|Timeline_0 PlayFromStart)))
              (else
                (Utilities|FlowControl|Sequence
                  (:then_0
                    (Timeline|Timeline_1 PlayFromStart))
                  (:then_1
                    (Timeline|Timeline_2 PlayFromStart))
                  (:then_2
                    _)
                  (:then_3
                    (CallFunction|ReplayCountDown)))))
            (bind _returnvalue_2 (PartModifier|SpawnSystem|SpawnSystemAttachedPartModifier "/BRCosmetics/Accessories/FORT_Backpacks/Backpack_F_MED_Ghost_Hunter/FX/NS_Backpack_GhostHunter.NS_Backpack_GhostHunter" (CustomCharacter|GetSkeletalMeshForPartOnPlayerPawn) "Root")
              (:Success
                (CallFunction|GhostHunterSetFXFrontEndFlag _returnvalue_2)))))))
    (:B
      ;;[1 0.575 0.2453 1 200] RIGHTSIDE
      (if (not (>= (Variables|Default|GetCounter_Right) 11.0))
        (bind _output_get_2 (Variables|Default|SetCounter_Right (+ (Variables|Default|GetCounter_Right) 1.0)))
        (bind _result_2 (ArrayMacros|SafeArrayGet (Variables|Default|GetvertexColor) (Math|Float|Truncate _output_get_2))
          (:%20%20
            (Rendering|Material|SetScalarParameterValue (Variables|Default|GetREF_BackpackMat) "RightCounter_Progress" _result_2 :!class /Script/Engine.MaterialInstanceDynamic)
            (if (>= (Variables|Default|GetCounter_Right) 1e+01)
              (Utilities|FlowControl|Sequence
                (:then_0
                  ;;;[1 0.4278 0.2058] ON KILL BEFORE OVERLAY
                  (Timeline|Timeline_4 PlayFromStart))
                (:then_1
                  (Timeline|Timeline_5 PlayFromStart))
                (:then_2
                  _)
                (:then_3
                  ;;;[1 1 1] OVERLAY
                  (Timeline|Timeline_3 PlayFromStart)))
              (else
                (Utilities|FlowControl|Sequence
                  (:then_0
                    (Timeline|Timeline_4 PlayFromStart))
                  (:then_1
                    (Timeline|Timeline_5 PlayFromStart))
                  (:then_2
                    _)
                  (:then_3
                    (CallFunction|ReplayCountDown)))))
            (bind _returnvalue_3 (PartModifier|SpawnSystem|SpawnSystemAttachedPartModifier "/BRCosmetics/Accessories/FORT_Backpacks/Backpack_F_MED_Ghost_Hunter/FX/NS_Backpack_GhostHunter.NS_Backpack_GhostHunter" (CustomCharacter|GetSkeletalMeshForPartOnPlayerPawn) "Root")
              (:Success
                (CallFunction|GhostHunterSetFXFrontEndFlag _returnvalue_3)))))))))

(event EventOnPawnAndPartChanged
  (bind _ref_backpackmat (Variables|Default|GetREF_BackpackMat))
  (bind _ref_backpackmat_2 (Variables|Default|GetREF_BackpackMat))
  (|CallParentFunction)
  (Utilities|FlowControl|Sequence
    (:then_0
      ;[1 0.2176 0.2574 1 200] ON KILL
      (Utilities|FlowControl|Sequence
        (:then_0
          ;;[1 1 1 1 200] LEFTSIDE
          (bind (_statname _statvalue) (Cosmetics|WatchCosmeticStat (CustomCharacter|GetAssociatedPlayerPawn) "(TagName=\"StatManager.AthenaKills\")" 0 false)
            (:OnStatChanged
              (Utilities|FlowControl|Delay 0.25)
              (if (> _statvalue 0)
                (bind _result (ArrayMacros|SafeArrayGet (Variables|Default|GetvertexColor) (Math|Integer|Clamp(Integer) _statvalue 0 10))
                  (:%20%20
                    (Rendering|Material|SetScalarParameterValue (Variables|Default|GetREF_BackpackMat) "LeftCounterProgress" _result :!class /Script/Engine.MaterialInstanceDynamic)
                    (if (>= (Math|Integer|Clamp(Integer) _statvalue 0 10) 10)
                      (if (>= _statvalue 11)
                        ;;;[1 1 1] OVERLAY
                        (bind (_direction _overlayintensity) (Timeline|Timeline_10 PlayFromStart)
                          (:Update
                            (Rendering|Material|SetScalarParameterValue (Variables|Default|GetREF_BackpackMat) "Overlay_Lerp" _overlayintensity :!class /Script/Engine.MaterialInstanceDynamic)))
                        (else
                          (Utilities|FlowControl|Sequence
                            (:then_0
                              ;;;[1 0.4278 0.2058] ON KILL BEFORE OVERLAY
                              (bind (_direction_2 _colorintensity) (Timeline|Timeline_9 PlayFromStart)
                                (:Update
                                  (Rendering|Material|SetScalarParameterValue _ref_backpackmat "Color_LeftSideMultiplier" _colorintensity :!class /Script/Engine.MaterialInstanceDynamic))))
                            (:then_1
                              (bind (_direction_3 _pipmove) (Timeline|Timeline_11 PlayFromStart)
                                (:Update
                                  (Rendering|Material|SetScalarParameterValue _ref_backpackmat "WPO_SelectedHeight" _pipmove :!class /Script/Engine.MaterialInstanceDynamic))))
                            (:then_2
                              _)
                            (:then_3
                              (Timeline|Timeline_10 PlayFromStart)))))
                      (else
                        (Utilities|FlowControl|Sequence
                          (:then_0
                            (Timeline|Timeline_9 PlayFromStart))
                          (:then_1
                            (Timeline|Timeline_11 PlayFromStart))
                          (:then_2
                            _))))
                    (PartModifier|SpawnSystem|SpawnSystemAttachedPartModifier "/BRCosmetics/Accessories/FORT_Backpacks/Backpack_F_MED_Ghost_Hunter/FX/NS_Backpack_GhostHunter.NS_Backpack_GhostHunter" (CustomCharacter|GetSkeletalMeshForPartOnPlayerPawn) "Root")))))))
        (:then_1
          ;;[1 0.575 0.2453 1 200] RIGHTSIDE
          (bind (_statname_2 _statvalue_2) (Cosmetics|WatchCosmeticStat (CustomCharacter|GetAssociatedPlayerPawn) "(TagName=\"StatManager.AIKills\")" 0 false)
            (:OnStatChanged
              (Utilities|FlowControl|Delay 0.25)
              (if (> _statvalue_2 0)
                (bind _result_2 (ArrayMacros|SafeArrayGet (Variables|Default|GetvertexColor) (Math|Integer|Clamp(Integer) _statvalue_2 0 10))
                  (:%20%20
                    (Rendering|Material|SetScalarParameterValue (Variables|Default|GetREF_BackpackMat) "RightCounter_Progress" _result_2 :!class /Script/Engine.MaterialInstanceDynamic)
                    (if (>= (Math|Integer|Clamp(Integer) _statvalue_2 0 10) 10)
                      (if (>= _statvalue_2 11)
                        ;;;[1 1 1] OVERLAY
                        (bind (_direction_4 _overlayintensity_2) (Timeline|Timeline_6 PlayFromStart)
                          (:Update
                            (Rendering|Material|SetScalarParameterValue (Variables|Default|GetREF_BackpackMat) "Overlay_Lerp" _overlayintensity_2 :!class /Script/Engine.MaterialInstanceDynamic)))
                        (else
                          (Utilities|FlowControl|Sequence
                            (:then_0
                              ;;;[1 0.4278 0.2058] ON KILL BEFORE OVERLAY
                              (bind (_direction_5 _colorintensity_2) (Timeline|Timeline_7 PlayFromStart)
                                (:Update
                                  (Rendering|Material|SetScalarParameterValue _ref_backpackmat_2 "Color_RightSideMultiplier" _colorintensity_2 :!class /Script/Engine.MaterialInstanceDynamic))))
                            (:then_1
                              (bind (_direction_6 _pipmove_2) (Timeline|Timeline_8 PlayFromStart)
                                (:Update
                                  (Rendering|Material|SetScalarParameterValue _ref_backpackmat_2 "WPO_SelectedHeight_RightSide" _pipmove_2 :!class /Script/Engine.MaterialInstanceDynamic))))
                            (:then_2
                              _)
                            (:then_3
                              (Timeline|Timeline_6 PlayFromStart)))))
                      (else
                        (Utilities|FlowControl|Sequence
                          (:then_0
                            (Timeline|Timeline_7 PlayFromStart))
                          (:then_1
                            (Timeline|Timeline_8 PlayFromStart))
                          (:then_2
                            _))))
                    (PartModifier|SpawnSystem|SpawnSystemAttachedPartModifier "/BRCosmetics/Accessories/FORT_Backpacks/Backpack_F_MED_Ghost_Hunter/FX/NS_Backpack_GhostHunter.NS_Backpack_GhostHunter" (CustomCharacter|GetSkeletalMeshForPartOnPlayerPawn) "Root")))))))))))

(fn GhostHunter_SetFXFrontEndFlag
  (params (NS :type Niagara%20Particle%20System%20Component%20Object%20Reference))
  (Niagara|SetNiagaraVariable(Bool) NS "bisFrontEndPreview" (CustomCosmeticModifier|IsinFrontEnd))
  (return))

(fn ConstructionScript
  (|CallParentFunction))
