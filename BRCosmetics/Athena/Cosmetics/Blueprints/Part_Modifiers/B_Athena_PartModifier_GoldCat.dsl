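; GoldCat character part modifier that turns the skin gold as kills rise and applies a Midas-touch wrap to weapons and vehicles.
; With the ReactiveOn tag, watched kill count drives superlevel material parameters, with a short refresh-timer workaround and a looping frontend preview.
; Tick feeds hand and chest socket positions to the material.
; The variant-selected wrap is applied to new items and entered vehicles, then cleared on drop or modifier destruction.
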(event Custom|CustomEvent
  (bind _sl_currentstage (Variables|Default|GetSL_CurrentStage))
  ; Frontend Preview
  (CallFunction|TransitionLevel (Math|Conversions|ToFloat(Integer) (Variables|Default|GetSL_CurrentStage)))
  (Math|Integer|IncrementInt _sl_currentstage
    (:%20%20
      (if (> _sl_currentstage (+ (Variables|Default|GetSL_EndLevel) 1))
        (Utilities|Time|ClearandInvalidateTimerbyHandle (stale _returnvalue))
        (Variables|Default|SetSL_CurrentStage 1)
        (CallFunction|TransitionLevel 0.0)
        (Utilities|FlowControl|Delay 3.0)
        (bind _returnvalue (Utilities|Time|SetTimerbyEvent (delegate Custom|CustomEvent) 4.0 true))
        (Variables|Default|SetFE_Loop_TimerHandle _returnvalue)))))

(event Custom|HackyWorkaroundRefreshTimerGuy
  ;[1 1 1] SuperLevelling Logic (turn into gold)
  (Utilities|FlowControl|Sequence
    (:then_0
      (CallFunction|InitializeVariabels)
      (Rendering|Material|SetScalarParameterValueonMaterials (CustomCharacter|GetSkeletalMeshForPartOnPlayerPawn) "SL_CurrentLevel" (Math|Conversions|ToFloat(Integer) (stale EventOnPostResIn _output_get)))
      (Utilities|FlowControl|DoOnce false
        (:Completed
          (bind _returnvalue (Utilities|Time|SetTimerbyEvent (delegate Custom|HackyWorkaroundRefreshTimerGuy) 0.1 true))
          (Variables|Default|SetSL_RefreshTimeHandler _returnvalue))))
    (:then_1
      (Utilities|FlowControl|Gate true
        (:Exit
          (if (and (> (Math|Integer|Max(Integer) (stale EventOnPostResIn _statvalue) (Variables|Default|GetSL_LastValue)) (Variables|Default|GetSL_StartLevel)) (< (Math|Integer|Max(Integer) (stale EventOnPostResIn _statvalue) (Variables|Default|GetSL_LastValue)) (+ (Variables|Default|GetSL_EndLevel) 1)))
            (CallFunction|OnActualStatChange)))))))

(event EventOnPostResIn
  (bind _sl_lastvalue (Variables|Default|GetSL_LastValue))
  ;[1 1 1] SuperLevelling Logic (turn into gold)
  (|CallParentFunction)
  ;; Clyde Superlevel
  (Rendering|Material|SetScalarParameterValueonMaterials (CustomCharacter|GetSkeletalMeshForPartOnPlayerPawn) "BP_UseCCPM" 1.0)
  (if (GameplayTags|HasTag (Cosmetics|GatherMetaTags (CustomCharacter|GetAssociatedPlayerPawn) "/BRCosmetics/Athena/Items/Cosmetics/Characters/Character_GoldCat_Claw.Character_GoldCat_Claw" :!class /Script/FortniteGame.FortPlayerPawn) "(TagName=\"Cosmetics.MetaTags.ReactiveOn\")")
    (CallFunction|ResetSafetyGate)
    (Utilities|FlowControl|Sequence
      (:then_0
        (Utilities|FlowControl|Delay 1.0)
        (Utilities|Time|ClearandInvalidateTimerbyHandle (Variables|Default|GetSL_RefreshTimeHandler)))
      (:then_1
        (bind (_statname _statvalue) (Cosmetics|WatchCosmeticStat (CustomCharacter|GetAssociatedPlayerPawn) "(TagName=\"StatManager.AthenaKills\")")
          (:OnStatChanged
            (bind _output_get (Variables|Default|SetSL_LastValue (Math|Integer|Max(Integer) _statvalue _sl_lastvalue)))
            (Utilities|FlowControl|Sequence
              (:then_0
                (CallFunction|InitializeVariabels)
                (Rendering|Material|SetScalarParameterValueonMaterials (CustomCharacter|GetSkeletalMeshForPartOnPlayerPawn) "SL_CurrentLevel" (Math|Conversions|ToFloat(Integer) _output_get))
                (Utilities|FlowControl|DoOnce false
                  (:Completed
                    (bind _returnvalue (Utilities|Time|SetTimerbyEvent (delegate Custom|HackyWorkaroundRefreshTimerGuy) 0.1 true))
                    (Variables|Default|SetSL_RefreshTimeHandler _returnvalue))))
              (:then_1
                (Utilities|FlowControl|Gate true
                  (:Exit
                    (if (and (> (Math|Integer|Max(Integer) _statvalue _sl_lastvalue) (Variables|Default|GetSL_StartLevel)) (< (Math|Integer|Max(Integer) _statvalue _sl_lastvalue) (+ (Variables|Default|GetSL_EndLevel) 1)))
                      (CallFunction|OnActualStatChange)))))))))
      (:then_2
        ; Frontend Preview
        (if (CustomCosmeticModifier|IsinFrontEnd)
          (Variables|Default|SetSL_CurrentStage 1)
          (CallFunction|TransitionLevel 0.0)
          (Utilities|FlowControl|Delay 3.0)
          (bind _returnvalue_2 (Utilities|Time|SetTimerbyEvent (delegate Custom|CustomEvent) 4.0 true))
          (Variables|Default|SetFE_Loop_TimerHandle _returnvalue_2))))
    (else
      (CallFunction|TransitionLevel 0.0))))

(event Custom|ResetSafetyGate
  ;[1 1 1] SuperLevelling Logic (turn into gold)
  (Utilities|FlowControl|DoOnce false
    (:Completed
      (bind _returnvalue (Utilities|Time|SetTimerbyEvent (delegate Custom|HackyWorkaroundRefreshTimerGuy) 0.1 true))
      (Variables|Default|SetSL_RefreshTimeHandler _returnvalue))))

(event Custom|OnActualStatChange
  ;[1 1 1] SuperLevelling Logic (turn into gold)
  ;;[1 1 1] This should only fire off when the stat changes - use this event!
  (Rendering|Material|SetScalarParameterValueonMaterials (CustomCharacter|GetSkeletalMeshForPartOnPlayerPawn) "SL_GameTimeFromBP" (Utilities|Time|GetGameTimeinSeconds)))

(event Custom|InitializeVariabels
  (bind _returnvalue (CustomCharacter|GetSkeletalMeshForPartOnPlayerPawn))
  ; Initialize Variables
  (Rendering|Material|SetScalarParameterValueonMaterials _returnvalue "SL_StartAtLevel" (Math|Conversions|ToFloat(Integer) (Variables|Default|GetSL_StartLevel)))
  (Rendering|Material|SetScalarParameterValueonMaterials _returnvalue "SL_StopAtLevel" (Math|Conversions|ToFloat(Integer) (Variables|Default|GetSL_EndLevel)))
  (Utilities|NotwhenBID)
  (Rendering|Material|SetScalarParameterValueonMaterials _returnvalue "SL_EndRadius" (Variables|Default|GetEndRadius)))

(event EventOnNewWeapons
  (params (WeaponsAsItems :type Array%20of%20Fort%20Item%20Object%20References))
  (bind _pendingitemstowrap (Variables|Default|GetPendingItemsToWrap))
  ;[0.9062 0.5249 0 0.5] Gather Weapons and Apply Midas Touch
  (|CallParentFunction WeaponsAsItems)
  (Utilities|NotwhenBID)
  (CallFunction|SelectWraptoUse)
  (for _array_element WeaponsAsItems
    (switch Utilities|FlowControl|Switch|SwitchonEFortItemType (Item|GetType _array_element)
      (:WeaponMelee
        _)
      (:WeaponRanged
        _))
    (Utilities|Array|Add (Variables|Default|GetPendingItemsToWrap) _array_element))
  (if (Utilities|IsValidSoftObjectReference (Variables|Default|GetWeaponWrapToUse))
    (Default|ApplyOverridetoItems _pendingitemstowrap)
    (Utilities|Array|Clear _pendingitemstowrap)))

(event EventOnPlayerPickupDropped
  (params (Pickup :type Fort%20Pickup%20Object%20Reference) (PickupItemEntry :type Fort%20Item%20Entry%20Structure%20%28by%20ref%29))
  ;[1 1 1] Force Weapons And Vehicles To Have A Specific Wrap (MIDAS TOUCH)
  ;;[0.4353 0.4275 0 0.502] Midas Touch Removal on Weapon Dropped
  (Utilities|IsValid Pickup
    (:Is%20Valid
      (if (== (Variables|Default|GetWeaponWrapToUse) (CallFunction|GetPrimaryItemEntryOverrideWrap PickupItemEntry))
        (Class|FortPickup|ClearOverrideWrapInternal Pickup)
        (bind _asfort_player_controller (CallFunction|GetAssociatedPlayerController))
        (Utilities|IsValid _asfort_player_controller
          (:Is%20Valid
            (Class|FortPlayerController|ServerClearOverrideWrap _asfort_player_controller Pickup)))))))

(event EventOnEnterVehicle
  (params (VehicleEntered :type Fort%20Vehicle%20Interface%20Interface%20%28by%20ref%29))
  ;[1 1 1] Force Weapons And Vehicles To Have A Specific Wrap (MIDAS TOUCH)
  (|CallParentFunction VehicleEntered)
  (Utilities|NotwhenBID)
  (if (Utilities|IsValidSoftObjectReference (Variables|Default|GetWeaponWrapToUse))
    (bind _asfort_player_controller (CallFunction|GetAssociatedPlayerController))
    (Utilities|IsValid _asfort_player_controller
      (:Is%20Valid
        (Class|FortPlayerController|ApplyOverrideWraptoVehicle _asfort_player_controller (Variables|Default|GetWeaponWrapToUse))))))

(event EventPreDestruction
  (bind _returnvalue (CustomCharacter|GetAssociatedPlayerPawn))
  (|CallParentFunction)
  (Utilities|Time|ClearandInvalidateTimerbyHandle (Variables|Default|GetFE_Loop_TimerHandle))
  (Utilities|Time|ClearandInvalidateTimerbyHandle (Variables|Default|GetSL_RefreshTimeHandler))
  ; Clear wrap when the part modifier is destroyed to ensure they are removed when switching character in the phone booth
  (Utilities|IsValid _returnvalue
    (:Is%20Valid
      (if (not (FortPawn|IsDead _returnvalue))
        (Class|FortPawn|ClearOverrideWrapfromItems _returnvalue (Variables|Default|GetWeaponWrapToUse))))))

(event EventTick
  (params (DeltaSeconds :type Float%20%28single-precision%29))
  (bind _playerpawnskeletalmesh (Variables|Default|GetPlayerPawnSkeletalMesh))
  (bind _returnvalue (Transformation|GetSocketLocation _playerpawnskeletalmesh "middle_metacarpal_l"))
  (bind _returnvalue_2 (Transformation|GetSocketLocation _playerpawnskeletalmesh "middle_metacarpal_r"))
  ; Clyde Superlevel Material (Butterfly effect)
  (Variables|Default|SetPlayerPawnSkeletalMesh (CustomCharacter|GetSkeletalMeshForPartOnPlayerPawn))
  (Rendering|Material|SetVectorParameterValueonMaterials _playerpawnskeletalmesh "BP_HandPosition_L" _returnvalue)
  (Rendering|Material|SetVectorParameterValueonMaterials _playerpawnskeletalmesh "BP_HandPosition_R" _returnvalue_2)
  (Rendering|Material|SetVectorParameterValueonMaterials _playerpawnskeletalmesh "BP_ChestPosition" (Transformation|GetSocketLocation _playerpawnskeletalmesh "spine_05"))
  (Rendering|Material|SetVectorParameterValueonMaterials _playerpawnskeletalmesh "BP_OffsetVector_L" (- _returnvalue (Transformation|GetSocketLocation _playerpawnskeletalmesh "middle_01_l")))
  (Rendering|Material|SetVectorParameterValueonMaterials _playerpawnskeletalmesh "BP_OffsetVector_R" (- _returnvalue_2 (Transformation|GetSocketLocation _playerpawnskeletalmesh "middle_01_r"))))

(fn ApplyOverrideToItems
  (params (Array :type Array%20of%20Fort%20Item%20Object%20References))
  (bind _asfort_player_controller (CallFunction|GetAssociatedPlayerController))
  (Utilities|IsValid _asfort_player_controller
    (:Is%20Valid
      (for _array_element Array
        (Class|FortPlayerController|ServerApplyOverrideWraptoItem _asfort_player_controller (Item|GetItemGuid _array_element) (Variables|Default|GetWeaponWrapToUse))))))

(fn CheckIfReactive
  (bind _returnvalue (CustomCharacter|GetAssociatedPlayerPawn))
  (if (Utilities|Enum|Equal(Enum) (CustomCharacter|GetCharacterPartType) "Backpack")
    ;[1 1 1] Backpack
    (bind _index (CosmeticLoadout|GetVariantChannelActiveIndexforSlotActor "(TagName=\"LoadoutSlot.Backpack\")" _returnvalue "(TagName=\"Cosmetics.Variant.Channel.Material\")" (Utilities|Array|MakeArray "(TagName=\"Cosmetics.Variant.Property.Mat1\")" "(TagName=\"Cosmetics.Variant.Property.Mat2\")" "(TagName=\"Cosmetics.Variant.Property.Mat3\")")))
    (switch int _index
      (:0
        (Variables|Default|SetbReactive true))
      (:1
        _)
      (:2
        _))
    (Variables|Default|SetbReactive false)
    (else
      (bind _index_2 (CosmeticLoadout|GetVariantChannelActiveIndexforSlotActor "(TagName=\"LoadoutSlot.Character\")" _returnvalue "(TagName=\"Cosmetics.Variant.Channel.Emissive\")" (Utilities|Array|MakeArray "(TagName=\"Cosmetics.Variant.Property.Emissive1\")" "(TagName=\"Cosmetics.Variant.Property.Emissive2\")")))
      (switch int _index_2
        (:0
          (Variables|Default|SetbReactive true))
        (:1
          (Variables|Default|SetbReactive false))))))

(fn GetAssociatedPlayerController
  (bind _returnvalue (CustomCharacter|GetAssociatedPlayerPawn))
  (Utilities|FlowControl|Sequence
    (:then_0
      (bind _aspawn (Utilities|Casting|CastToPawn (Vehicle|GetVehicleActor _returnvalue :!class /Script/FortniteGame.FortPlayerPawn)))
      (bind _asfort_player_controller (Utilities|Casting|CastToFortPlayerController (Pawn|GetController _aspawn)))
      (return _asfort_player_controller))
    (:then_1
      (bind _asfort_player_controller_2 (Utilities|Casting|CastToFortPlayerController (Pawn|GetController _returnvalue)))
      (return _asfort_player_controller_2))))

(fn SelectWrapToUse
  (bind (_index _bindexisvalid) (CosmeticLoadout|GetVariantChannelActiveIndexforSlotActor "(TagName=\"LoadoutSlot.Character\")" (CustomCharacter|GetAssociatedPlayerPawn) "(TagName=\"Cosmetics.Variant.Channel.Pattern\")" (Utilities|Array|MakeArray "(TagName=\"Cosmetics.Variant.Property.Stage4\")")))
  (Variables|Default|SetWeaponWrapToUse (select _bindexisvalid "/Game/Athena/Items/Cosmetics/ItemWraps/Wrap_CatBurglarGameplay.Wrap_CatBurglarGameplay" "/BRCosmetics/Athena/Items/Cosmetics/ItemWraps/Wrap_GoldCat.Wrap_GoldCat")))

(fn TransitionLevel
  (params (ParameterValue :type Float%20%28single-precision%29))
  (bind _returnvalue (CustomCharacter|GetSkeletalMeshForPartOnPlayerPawn))
  (Rendering|Material|SetScalarParameterValueonMaterials _returnvalue "SL_CurrentLevel" ParameterValue)
  (Rendering|Material|SetScalarParameterValueonMaterials _returnvalue "SL_GameTimeFromBP" (Utilities|Time|GetGameTimeinSeconds)))

(fn ConstructionScript
  (bind _returnvalue (CustomCharacter|GetSkeletalMeshForPartOnPlayerPawn))
  (|CallParentFunction)
  (Rendering|Material|SetScalarParameterValueonMaterials (CustomCharacter|GetSkeletalMeshForPartOnPlayerPawn) "SL_GameTimeFromBP" -99999.0)
  (Rendering|Material|SetScalarParameterValueonMaterials _returnvalue "SL_StartAtLevel" (Math|Conversions|ToFloat(Integer) (Variables|Default|GetSL_StartLevel)))
  (Rendering|Material|SetScalarParameterValueonMaterials _returnvalue "SL_StopAtLevel" (Math|Conversions|ToFloat(Integer) (Variables|Default|GetSL_EndLevel))))
