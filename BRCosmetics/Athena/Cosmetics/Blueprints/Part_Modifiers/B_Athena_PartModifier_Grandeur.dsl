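; Cosmetic part modifier Blueprint for Grandeur that animates a color tint on the character part's skeletal mesh materials in the front end.
; On post res-in, it enables useDebugScalarForColorTint and starts the ReactivityLoop timeline, which drives the DebugColorSwap scalar.
; The timeline repeats by calling RestartLoop whenever it finishes.
; Outside the front end, both material scalars are set to zero and the loop is not started.
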(event Custom|RestartLoop
  (bind (_direction _newtrack_0) (Timeline|ReactivityLoop PlayFromStart)
    (:Update
      (Rendering|Material|SetScalarParameterValueonMaterials (CustomCharacter|GetSkeletalMeshForPartOnPlayerPawn) "DebugColorSwap" _newtrack_0))
    (:Finished
      (CallFunction|RestartLoop))))

(event EventOnPostResIn
  (bind _returnvalue (CustomCharacter|GetSkeletalMeshForPartOnPlayerPawn))
  (if (CustomCosmeticModifier|IsinFrontEnd)
    (Rendering|Material|SetScalarParameterValueonMaterials _returnvalue "useDebugScalarForColorTint" 1.0)
    (Timeline|ReactivityLoop PlayFromStart)
    (else
      (Rendering|Material|SetScalarParameterValueonMaterials _returnvalue "useDebugScalarForColorTint" 0.0)
      (Rendering|Material|SetScalarParameterValueonMaterials _returnvalue "DebugColorSwap" 0.0))))

(fn ConstructionScript
  (|CallParentFunction))
