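; Character part modifier for the GreatPoolSwim cosmetic that turns off the eye effect on a specific progressive variant stage.
; After the parent post-res-in logic runs, it checks whether the associated player pawn's character loadout has Progressive Stage3 active.
; If so, and the idle VFX component is valid, it casts that component to Niagara and sets its bShowEyeFX parameter to false.
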(event EventOnPostResIn
  (bind _returnvalue (CallFunction|GetIdleVFXComponent))
  ;[0.3897 1 0] Show eye fx flag
  (|CallParentFunction)
  (bind (_index _bindexisvalid) (CosmeticLoadout|GetVariantChannelActiveIndexforSlotActor "(TagName=\"LoadoutSlot.Character\")" (CustomCharacter|GetAssociatedPlayerPawn) "(TagName=\"Cosmetics.Variant.Channel.Progressive\")" (Utilities|Array|MakeArray "(TagName=\"Cosmetics.Variant.Property.Stage3\")") false false))
  (if _bindexisvalid
    (Utilities|IsValid _returnvalue
      (:Is%20Valid
        (bind _asniagara_particle_system_component (Utilities|Casting|CastToNiagaraComponent _returnvalue))
        (Niagara|SetNiagaraVariable(Bool) _asniagara_particle_system_component "bShowEyeFX" false)))))

(fn ConstructionScript
  (|CallParentFunction))
