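; Cosmetic part modifier Blueprint for the GreatPool backpack that recolors its idle particle effect to match the selected material variant.
; After calling the parent's post-res-in event, it reads the active Material channel index among three variants from the backpack slot of the cosmetic loadout.
; That index picks a yellow, blue or red pair for the Glow Color and Eyes Color parameters, with red as the fallback.
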(event EventOnPostResIn
  (bind _returnvalue (CallFunction|GetIdleVFXComponent))
  ;[1 0.2176 0.2574 1 200] EVENT 🖥️
  (|CallParentFunction)
  (Utilities|FlowControl|Sequence
    (:then_2
      (bind _index (Class|FortCosmeticFunctionLibrary|GetVariantChannelActiveIndexForSlot "(TagName=\"LoadoutSlot.Backpack\")" (CosmeticLoadout|GetCosmeticLoadout) "(TagName=\"Cosmetics.Variant.Channel.Material\")" (Utilities|Array|MakeArray "(TagName=\"Cosmetics.Variant.Property.Mat1\")" "(TagName=\"Cosmetics.Variant.Property.Mat2\")" "(TagName=\"Cosmetics.Variant.Property.Mat3\")")))
      (if (== _index 0)
        (Effects|Components|ParticleSystem|SetColorParameter _returnvalue "Glow Color" "(R=0.991102,G=0.982251,B=0.001518,A=0.700000)")
        (Effects|Components|ParticleSystem|SetColorParameter _returnvalue "Eyes Color" "(R=8.000000,G=7.928553,B=0.012250,A=1.000000)")
        (elif (== _index 1)
          (Effects|Components|ParticleSystem|SetColorParameter _returnvalue "Glow Color" "(R=0.104573,G=0.438952,B=1.500000,A=1.000000)")
          (Effects|Components|ParticleSystem|SetColorParameter _returnvalue "Eyes Color" "(R=2.084812,G=4.715775,B=15.000000,A=1.000000)")
          (else
            (Effects|Components|ParticleSystem|SetColorParameter _returnvalue "Glow Color" "(R=1.000000,G=0.003631,B=0.004367,A=1.000000)")
            (Effects|Components|ParticleSystem|SetColorParameter _returnvalue "Eyes Color" "(R=15.000000,G=0.054470,B=0.065502,A=1.000000)")))))))

(fn ConstructionScript
  (|CallParentFunction))
