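; GrilledCheese cosmetic part modifier that shows four escalating visual stages on a character as cosmetic state activates.
; After res-in it caches head, body and face meshes; in the front end, a timeline previews the stages.
; Stage events drive emissive and skin material parameters across all parts, plus face-only fresnel and brightness, with FinalStage toggling idle VFX.
; Stage activation also spawns an attached Niagara burst, passed an IsFrontEnd flag.
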(event EventOnPostResIn
  (|CallParentFunction)
  (CallFunction|GetParts)
  (if (CustomCosmeticModifier|IsinFrontEnd)
    (Timeline|Stage%20Cycle%20Transition PlayFromStart
      (:Stage%201
        (CallFunction|StageOne))
      (:Stage%202
        (CallFunction|StageTwo))
      (:Stage%203
        (CallFunction|StageThree))
      (:Stage%204
        (CallFunction|StageFour)))
    (bind _returnvalue (PartModifier|SpawnSystem|SpawnSystemAttachedPartModifier "/BRCosmetics/Effects/Fort_Effects/Effects/Characters/Athena_Parts/GrilledCheese/NS_GrilledCheese_Burst.NS_GrilledCheese_Burst" (CustomCharacter|GetSkeletalMeshForPartOnPlayerPawn) "root")
      (:Success
        (Niagara|SetNiagaraVariable(Bool) _returnvalue "IsFrontEnd" (CustomCosmeticModifier|IsinFrontEnd))))))

(event EventOnCosmeticStateActivated
  (params (Stat%20Name :type Name :default "None") (Stat%20Value :type Integer) (STAGE%20Num%20Being%20Activated :type Integer))
  (|CallParentFunction Stat%20Name Stat%20Value STAGE%20Num%20Being%20Activated)
  (switch int STAGE%20Num%20Being%20Activated
    (:0
      (CallFunction|StageOne))
    (:1
      (CallFunction|StageTwo))
    (:2
      (CallFunction|StageThree))
    (:3
      (CallFunction|StageFour)))
  (bind _returnvalue (PartModifier|SpawnSystem|SpawnSystemAttachedPartModifier "/BRCosmetics/Effects/Fort_Effects/Effects/Characters/Athena_Parts/GrilledCheese/NS_GrilledCheese_Burst.NS_GrilledCheese_Burst" (CustomCharacter|GetSkeletalMeshForPartOnPlayerPawn) "root")
    (:Success
      (Niagara|SetNiagaraVariable(Bool) _returnvalue "IsFrontEnd" (CustomCosmeticModifier|IsinFrontEnd)))))

(event Custom|StageOne
  (bind _face (Variables|Default|Getface))
  (Effects|Components|ParticleSystem|SetFloatParameter (CallFunction|GetIdleVFXComponent) "FinalStage" 0.0)
  ;[1 1 1] TechArt Material Changes
  (CallFunction|SetScalarParameteronMaterialsMultiParts "EmissiveReveal")
  (CallFunction|SetScalarParameteronMaterialsMultiParts "ColorSkinStrength")
  ;[1 1 1] hair
  (Rendering|Material|SetScalarParameterValueonMaterials _face "FresnelTintStrength" 0.0)
  (Rendering|Material|SetScalarParameterValueonMaterials _face "FresnelTintExponent" 1.5)
  (Rendering|Material|SetScalarParameterValueonMaterials _face "Brightness_Multiplier" 1.0))

(event Custom|StageTwo
  (bind _face (Variables|Default|Getface))
  ;[1 1 1] TechArt Material Changes
  (CallFunction|SetScalarParameteronMaterialsMultiParts "EmissiveReveal" 1.0)
  (CallFunction|SetScalarParameteronMaterialsMultiParts "ColorSkinStrength")
  (CallFunction|SetScalarParameteronMaterialsMultiParts "Brightness_Min" 0.2)
  (CallFunction|SetScalarParameteronMaterialsMultiParts "Brightness_Max" 5.0)
  (CallFunction|SetScalarParameteronMaterialsMultiParts "SkinEmissiveStrength")
  ;[1 1 1] hair
  (Rendering|Material|SetScalarParameterValueonMaterials _face "FresnelTintStrength" 0.33)
  (Rendering|Material|SetScalarParameterValueonMaterials _face "FresnelTintExponent" 2.0)
  (Rendering|Material|SetScalarParameterValueonMaterials _face "Brightness_Multiplier" 2.5))

(event Custom|StageThree
  (bind _face (Variables|Default|Getface))
  ;[1 1 1] TechArt Material Changes
  (CallFunction|SetScalarParameteronMaterialsMultiParts "EmissiveReveal" 1.0)
  (CallFunction|SetScalarParameteronMaterialsMultiParts "ColorSkinStrength" 1.0)
  (CallFunction|SetVectorParameterOnMaterialsMultiParts "SkinColor" "0.600000,0.050000,0.020000")
  (CallFunction|SetScalarParameteronMaterialsMultiParts "Brightness_Min" 0.2)
  (CallFunction|SetScalarParameteronMaterialsMultiParts "Brightness_Max" 5.0)
  (CallFunction|SetScalarParameteronMaterialsMultiParts "SkinEmissiveStrength" 0.1)
  ;[1 1 1] hair
  (Rendering|Material|SetScalarParameterValueonMaterials _face "FresnelTintStrength" 0.66)
  (Rendering|Material|SetScalarParameterValueonMaterials _face "FresnelTintExponent" 2.0)
  (Rendering|Material|SetScalarParameterValueonMaterials _face "Brightness_Multiplier" 5.0))

(event Custom|StageFour
  (bind _face (Variables|Default|Getface))
  (Effects|Components|ParticleSystem|SetFloatParameter (CallFunction|GetIdleVFXComponent) "FinalStage" 1.0)
  ;[1 1 1] TechArt Material Changes
  (CallFunction|SetScalarParameteronMaterialsMultiParts "EmissiveReveal" 1.0)
  (CallFunction|SetScalarParameteronMaterialsMultiParts "ColorSkinStrength" 1.0)
  (CallFunction|SetVectorParameterOnMaterialsMultiParts "SkinColor" "0.700000,0.020000,0.010000")
  (CallFunction|SetScalarParameteronMaterialsMultiParts "Brightness_Min" 0.6)
  (CallFunction|SetScalarParameteronMaterialsMultiParts "Brightness_Max" 5.5)
  (CallFunction|SetScalarParameteronMaterialsMultiParts "SkinEmissiveStrength" 0.25)
  ;[1 1 1] hair
  (Rendering|Material|SetScalarParameterValueonMaterials _face "FresnelTintStrength" 1.0)
  (Rendering|Material|SetScalarParameterValueonMaterials _face "FresnelTintExponent" 1.5)
  (Rendering|Material|SetScalarParameterValueonMaterials _face "Brightness_Multiplier" 4.0))

(fn GetParts
  (bind _returnvalue (CustomCharacter|GetAssociatedPlayerPawn))
  (Variables|Default|Sethead (Customization|GetSkeletalMeshforPartType _returnvalue "Head" :!class /Script/FortniteGame.FortPlayerPawn))
  (Variables|Default|Setbody (Customization|GetSkeletalMeshforPartType _returnvalue "Body" :!class /Script/FortniteGame.FortPlayerPawn))
  (Variables|Default|Setface (Customization|GetSkeletalMeshforPartType _returnvalue "Face" :!class /Script/FortniteGame.FortPlayerPawn)))

(fn SetScalarParameterOnMaterialsMultiParts
  (params (name :type Name) (value :type Float%20%28double-precision%29))
  (Rendering|Material|SetScalarParameterValueonMaterials (Variables|Default|Gethead) name value)
  (Rendering|Material|SetScalarParameterValueonMaterials (Variables|Default|Getface) name value)
  (Rendering|Material|SetScalarParameterValueonMaterials (Variables|Default|Getbody) name value))

(fn SetVectorParameteronMaterialsMultiParts
  (params (name :type Name) (value :type Vector))
  (Rendering|Material|SetVectorParameterValueonMaterials (Variables|Default|Gethead) name value)
  (Rendering|Material|SetVectorParameterValueonMaterials (Variables|Default|Getface) name value)
  (Rendering|Material|SetVectorParameterValueonMaterials (Variables|Default|Getbody) name value))

(fn ConstructionScript
  (|CallParentFunction))
