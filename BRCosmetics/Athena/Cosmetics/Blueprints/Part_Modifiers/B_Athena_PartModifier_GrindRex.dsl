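; Character part modifier Blueprint for the GrindRex cosmetic that drives reactive material parameters on the part's skeletal mesh.
; On post-resolution init in a frontend context, it sets a front-end flag and current game time on the materials, then repeats this every 12 seconds on a looping timer.
; Accumulated velocity events feed normalized velocity into the running parameter, so the materials respond to player movement.
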(event EventOnPostResIn
  (|CallParentFunction)
  (if (Reactive|Frontend|IsCosmeticObjectDisplayedinFrontendContext 0)
    (CallFunction|FrontEndPreview)
    (Utilities|Time|SetTimerbyEvent (delegate Custom|FrontEndPreview) 12.0 true)))

(event Custom|FrontEndPreview
  (bind _returnvalue (CustomCharacter|GetAssociatedPlayerPawn))
  (bind _returnvalue_2 (CustomCharacter|GetCharacterPartType))
  (bind _returnvalue_3 (Customization|GetSkeletalMeshforPartType _returnvalue _returnvalue_2 :!class /Script/FortniteGame.FortPlayerPawn))
  (Rendering|Material|SetScalarParameterValueonMaterials _returnvalue_3 "BP_IsFrontEnd" 1.0)
  (Rendering|Material|SetScalarParameterValueonMaterials _returnvalue_3 "BP_GetTime" (Utilities|Time|GetGameTimeinSeconds)))

(event EventOnAccumulatedVelocity
  (params (NormalizedAccumulatedVelocity :type Float%20%28double-precision%29) (AccumulatedVelocity :type Float%20%28double-precision%29))
  (Rendering|Material|SetScalarParameterValueonMaterials (Customization|GetSkeletalMeshforPartType (CustomCharacter|GetAssociatedPlayerPawn) (CustomCharacter|GetCharacterPartType) :!class /Script/FortniteGame.FortPlayerPawn) "BP_IsRunning" NormalizedAccumulatedVelocity))

(fn ConstructionScript)
