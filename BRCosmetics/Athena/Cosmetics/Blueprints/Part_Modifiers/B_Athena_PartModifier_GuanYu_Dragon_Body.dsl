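; Cosmetic part modifier for the Guan Yu Dragon body that adds Niagara fire effects to the character's body part mesh.
; On post res-in it attaches ankle fire and left pauldron fire, plus an inactive pauldron burst.
; A watched cosmetic stat reactivates the burst whenever it changes to a value above zero.
; Weapon changes run a hidden-part check on the feet HideFX tag to drive the ankle fire's foot-hidden parameter.
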(event EventOnPostResIn
  (bind _returnvalue (CustomCharacter|GetCharacterPartType))
  (bind _returnvalue_2 (CustomCharacter|GetAssociatedPlayerPawn))
  (bind _returnvalue_3 (Customization|GetSkeletalMeshforPartType _returnvalue_2 _returnvalue :!class /Script/FortniteGame.FortPlayerPawn))
  (|CallParentFunction)
  (bind (_statname _statvalue) (Cosmetics|WatchCosmeticStat _returnvalue_2 (Variables|Game|GetCheckStatName))
    (:then
      (bind _returnvalue_4 (Niagara|SpawnSystemAttached (Variables|Default|GetNS_GuanYu_Dragon_AnkleFire) _returnvalue_3 "fx_root"))
      (Niagara|SetNiagaraVariable(Bool) _returnvalue_4 "bisFootHidden" (stale EventOnWeaponChanged _isparthidden))
      (Niagara|SpawnSystemAttached "/BRCosmetics/Characters/Player/Male/Medium/Bodies/M_MED_GuanYuDragon/VFX/NS_GuanYu_Dragon_Pauldron_Fire.NS_GuanYu_Dragon_Pauldron_Fire" _returnvalue_3 "fx_dragonPauldron_l")
      (bind _returnvalue_5 (Niagara|SpawnSystemAttached (Variables|Default|GetNS_GuanYu_Dragon_Pauldron_Burst) _returnvalue_3 "fx_dragonPauldron_l" "0, 0, 0" "0, 0, 0" "KeepRelativeOffset" false false)))
    (:OnStatChanged
      (if (> _statvalue 0)
        (CallFunction|ReactiveEffect)))))

(event Custom|Reactive%20Effect
  (Components|Activation|Activate (stale EventOnPostResIn _returnvalue_5) true))

(event EventOnWeaponChanged
  (params (NewWeapon :type Fort%20Weapon%20Object%20Reference) (OldWeapon :type Fort%20Weapon%20Object%20Reference))
  (|CallParentFunction NewWeapon OldWeapon)
  (bind _isparthidden (Default|HiddenPartCheck-Weapon (Utilities|Array|MakeArray (stale EventOnPostResIn _returnvalue_4)) NewWeapon OldWeapon true "(TagName=\"Animation.CharacterParts.HideFX.Feet\")" "bisFootHidden")))

(fn ConstructionScript)
