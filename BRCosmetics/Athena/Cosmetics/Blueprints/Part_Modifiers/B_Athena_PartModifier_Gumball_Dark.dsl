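; Cosmetic part modifier Blueprint for the dark Gumball character, tuning material transparency on the character part mesh.
; On the post-res-in event it fetches the part's skeletal mesh on the player pawn.
; It sets BottomTransparency and RimTransparency scalar parameters on that mesh's materials, using lower values in the front end than in matches.
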(event EventOnPostResIn
  (bind _returnvalue (CustomCharacter|GetSkeletalMeshForPartOnPlayerPawn))
  (if (CustomCosmeticModifier|IsinFrontEnd)
    (Rendering|Material|SetScalarParameterValueonMaterials _returnvalue "BottomTransparency" 0.25)
    (Rendering|Material|SetScalarParameterValueonMaterials _returnvalue "RimTransparency" 0.75)
    (else
      (Rendering|Material|SetScalarParameterValueonMaterials _returnvalue "BottomTransparency" 0.5)
      (Rendering|Material|SetScalarParameterValueonMaterials _returnvalue "RimTransparency" 1.0))))

(fn ConstructionScript
  (|CallParentFunction))
