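; Cosmetic part modifier for the HandyMan character that drives smoke and idle VFX when targeting or first-person camera state changes.
; When the pawn is targeting outside the front end, it activates the idle VFX component and spawns attached Niagara reload smoke at the part mesh's FX_SmokeR socket.
; A separate step deactivates the idle VFX component.
; Both steps sit behind their own DoOnce gates, with no reset shown.
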(event EventOnTargetingOrFirstPersonCameraChanged
  (bind _returnvalue (CallFunction|GetIdleVFXComponent))
  (|CallParentFunction)
  (if (and (Variables|/Script/FortniteGame.FortPlayerPawn|GetbIsTargeting (CustomCharacter|GetAssociatedPlayerPawn)) (not (CustomCosmeticModifier|IsinFrontEnd)))
    (Utilities|FlowControl|DoOnce false
      (:Completed
        (Components|Activation|Activate _returnvalue true)
        (PartModifier|SpawnSystem|SpawnSystemAttachedPartModifier "/BRCosmetics/Athena/Effects/Particles/Skins/HandyMan/NS_HandyMan_ReloadSmoke.NS_HandyMan_ReloadSmoke" (CustomCharacter|GetSkeletalMeshForPartOnPlayerPawn) "FX_SmokeR"
          (:Success
            _)
          (:Failure
            _))))
    (else
      _))
  (Utilities|FlowControl|DoOnce false
    (:Completed
      (Components|Activation|Deactivate _returnvalue))))

(fn ConstructionScript
  (|CallParentFunction))
