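; Cosmetic part modifier for the HarborSnow character that drives a reactive battery logo on its materials.
; On post res-in, the ReactiveOn meta tag toggles BatteryOpacity.
; When reactive, the body part triggers the logo animation, with a front-end delay path, and chains watches on chest-open, shield-gain and health-gain stats.
; The animation is material-driven: the trigger event writes current game time into Reactivity_TL.
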(event EventOnDayPhaseChanged
  (params (CurrentDayPhase :type EFortDayPhase%20Enum :default "Morning") (PreviousDayPhase :type EFortDayPhase%20Enum :default "Morning") (bAtCreation :type Boolean)))

(event EventOnNewPlayerWeaponType
  (params (NewType :type EFortItemType%20Enum :default "WorldItem") (OldType :type EFortItemType%20Enum :default "WorldItem"))
  (|CallParentFunction NewType OldType) ; This node is disabled and will not be called.\nDrag off pins to build functionality.
)

(event EventOnWeaponChanged
  (params (NewWeapon :type Fort%20Weapon%20Object%20Reference) (OldWeapon :type Fort%20Weapon%20Object%20Reference))
  (|CallParentFunction NewWeapon OldWeapon) ; This node is disabled and will not be called.\nDrag off pins to build functionality.
)

(event EventOnWeaponFired)

(event EventOnNewWeapons
  (params (WeaponsAsItems :type Array%20of%20Fort%20Item%20Object%20References)))

(event EventOnCurieStateAttached
  (params (Tag :type Gameplay%20Tag%20Structure%20%28by%20ref%29)))

(event EventOnCurieStateDetached
  (params (Tag :type Gameplay%20Tag%20Structure%20%28by%20ref%29)))

(event EventOnBeginSkydiving
  (|CallParentFunction) ; This node is disabled and will not be called.\nDrag off pins to build functionality.
)

(event EventOnEndSkydiving
  (|CallParentFunction) ; This node is disabled and will not be called.\nDrag off pins to build functionality.
)

(event EventOnEnterVehicle
  (params (VehicleEntered :type Fort%20Vehicle%20Interface%20Interface%20%28by%20ref%29))
  (|CallParentFunction VehicleEntered) ; This node is disabled and will not be called.\nDrag off pins to build functionality.
)

(event EventOnExitVehicle
  (|CallParentFunction) ; This node is disabled and will not be called.\nDrag off pins to build functionality.
)

(event EventOnTargetingOrFirstPersonCameraChanged)

(event EventOnHitPawn
  (params (SelfActor :type Actor%20Object%20Reference) (OtherActor :type Actor%20Object%20Reference) (NormalImpulse :type Vector :default "0, 0, 0") (Hit :type Hit%20Result%20Structure%20%28by%20ref%29)))

(event EventOnPostResIn
  (bind _returnvalue (CustomCharacter|GetSkeletalMeshForPartOnPlayerPawn))
  (bind _returnvalue_2 (CustomCharacter|GetAssociatedPlayerPawn))
  (|CallParentFunction)
  (if (GameplayTags|HasTag (Cosmetics|GatherMetaTags (CustomCharacter|GetAssociatedPlayerPawn) "/BRCosmetics/Athena/Items/Cosmetics/Characters/Character_HarborSnow.Character_HarborSnow" :!class /Script/FortniteGame.FortPlayerPawn) "(TagName=\"Cosmetics.MetaTags.ReactiveOn\")")
    (Rendering|Material|SetScalarParameterValueonMaterials _returnvalue "BatteryOpacity" 1.0)
    ;[0.3629 0.1311 1] Watch Attributs
    (if (Utilities|Enum|Equal(Enum) (CustomCharacter|GetCharacterPartType) "Body")
      (if (CustomCosmeticModifier|IsinFrontEnd)
        (CallFunction|TriggerBatteryLogoAnim)
        (Utilities|FlowControl|Delay 12.0))
      (Cosmetics|WatchCosmeticStat _returnvalue_2 "(TagName=\"StatManager.OpenChests\")" 0 false
        (:then
          (Cosmetics|WatchCosmeticStat _returnvalue_2 "(TagName=\"StatManager.GainedShieldTimes\")" 0 false
            (:then
              (Cosmetics|WatchCosmeticStat _returnvalue_2 "(TagName=\"StatManager.GainedHealthTimes\")" 0 false
                (:OnStatChanged
                  _)))
            (:OnStatChanged
              _)))
        (:OnStatChanged
          _))
      (CallFunction|TriggerBatteryLogoAnim))
    (else
      (Rendering|Material|SetScalarParameterValueonMaterials _returnvalue "BatteryOpacity" 0.0))))

(event Custom|TriggerBatteryLogoAnim
  ;[0 0.9911 1] Material Charge Animation
  (Rendering|Material|SetScalarParameterValueonMaterials (CustomCharacter|GetSkeletalMeshForPartOnPlayerPawn) "Reactivity_TL" (Utilities|Time|GetGameTimeinSeconds)))

(fn ConstructionScript
  (|CallParentFunction))
