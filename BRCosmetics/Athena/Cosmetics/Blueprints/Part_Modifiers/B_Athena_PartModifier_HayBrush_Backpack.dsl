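; Cosmetic part modifier for the HayBrush backpack, providing frontend and shield-gain reactivity.
; After init, frontend display waits, then writes game time to the mesh's BP_GetGameTime material parameter, with no repeat despite the loop name.
; Otherwise it watches the GainedShieldTimes stat, but the ShieldIncrease handler is empty.
; Helper functions, uncalled here, arm shield audio on quick-healing tags and play it once, attached to the mesh.
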(event Custom|FrontEndReactivityLoop
  ;[0.4766 0.4867 1] FrontEnd Reactivity Loop
  (Utilities|FlowControl|Delay 1e+01)
  (Rendering|Material|SetScalarParameterValueonMaterials (CustomCharacter|GetSkeletalMeshForPartOnPlayerPawn) "BP_GetGameTime" (Utilities|Time|GetGameTimeinSeconds)))

(event EventOnPostResIn
  (|CallParentFunction)
  (if (Reactive|Frontend|ShouldDisplayFrontendReactivity 0 "Backpack")
    (CallFunction|FrontEndReactivityLoop)
    (else
      (Cosmetics|WatchCosmeticStat (CustomCharacter|GetAssociatedPlayerPawn) "(TagName=\"StatManager.GainedShieldTimes\")" 0 false
        (:OnStatChanged
          (CallFunction|ShieldIncrease))))))

(event Custom|ShieldIncrease)

(fn AllowShieldAudio
  (params (GameplayAbility :type Fort%20Gameplay%20Ability%20Object%20Reference) (AbilityTags :type Gameplay%20Tag%20Container%20Structure))
  (if (GameplayTags|MatchesAnyTags "(TagName=\"Gameplay.Action.Player.QuickHealingItem\")" AbilityTags true)
    (Variables|Default|SetbAllowReactiveAudio true)))

(fn PlayShieldReactiveAudio
  (if (Variables|Default|GetbAllowReactiveAudio)
    (Audio|SpawnSoundAttached (Variables|Default|GetShieldGainSoundCue) (CustomCharacter|GetSkeletalMeshForPartOnPlayerPawn))
    (Variables|Default|SetbAllowReactiveAudio false)))

(fn ConstructionScript)
