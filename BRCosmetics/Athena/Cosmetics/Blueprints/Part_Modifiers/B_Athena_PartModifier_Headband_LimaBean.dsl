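; Character part modifier Blueprint for the LimaBean headband cosmetic, which tells its audio whether the wearer is locally viewed.
; UpdateAudioStatValue gets the associated player pawn and, if valid, writes its IsLocallyViewed result to the "IsLocal" boolean parameter on the first supplied audio component.
; The name and stat value inputs go unused, and the construction script only calls its parent.
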(fn UpdateAudioStatValue
  (params (AudioComponent :type Array%20of%20Audio%20Component%20Object%20References) (InName :type Name :default "None") (StatValue :type Integer))
  (bind _returnvalue (CustomCharacter|GetAssociatedPlayerPawn))
  (Utilities|IsValid _returnvalue
    (:Is%20Valid
      (Audio|Parameter|SetBooleanParameter (Utilities|Array|Get(acopy) AudioComponent) "IsLocal" (Pawn|IsLocallyViewed _returnvalue) :!class /Script/AudioExtensions.AudioParameterControllerInterface)
      (return))))

(fn ConstructionScript
  (|CallParentFunction))
