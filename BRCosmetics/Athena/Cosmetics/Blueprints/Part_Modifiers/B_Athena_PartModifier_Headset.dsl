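; Part modifier for the Headset cosmetic's character parts and backpack, giving reactive styles flame and transformation stages that advance with player kills.
; In match it watches the kill stat and switches on stages 0–5, driving material reveal, emissive and element-hiding parameters plus Niagara bursts.
; A part-applied callback caches the mesh and sets idle VFX.
; The front end previews stages via timeline; non-reactive styles get fixed looks.
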(event Custom|CE_Frontend%20FX%20Loop%20Burst)

(event EventOnPostResIn
  (bind _returnvalue (CustomCharacter|GetSkeletalMeshForPartOnPlayerPawn))
  (bind _returnvalue_2 (CustomCharacter|GetSkeletalMeshForPartOnPlayerPawn))
  (bind _returnvalue_3 (CustomCharacter|GetSkeletalMeshForPartOnPlayerPawn))
  (bind _returnvalue_4 (CustomCharacter|GetSkeletalMeshForPartOnPlayerPawn))
  (bind _returnvalue_5 (CustomCharacter|GetSkeletalMeshForPartOnPlayerPawn))
  (bind _returnvalue_6 (CustomCharacter|GetSkeletalMeshForPartOnPlayerPawn))
  (bind _returnvalue_7 (CustomCharacter|GetSkeletalMeshForPartOnPlayerPawn))
  (bind _returnvalue_8 (CustomCharacter|GetAssociatedPlayerPawn))
  (bind _returnvalue_9 (Customization|GetSkeletalMeshforPartType _returnvalue_8 "Backpack" :!class /Script/FortniteGame.FortPlayerPawn))
  (bind _returnvalue_10 (CustomCharacter|GetAssociatedPlayerPawn))
  (bind _returnvalue_11 (CustomCharacter|GetSkeletalMeshForPartOnPlayerPawn))
  (bind _returnvalue_12 (CustomCharacter|GetSkeletalMeshForPartOnPlayerPawn))
  (bind _returnvalue_13 (CustomCharacter|GetSkeletalMeshForPartOnPlayerPawn))
  (bind _returnvalue_14 (CustomCharacter|GetSkeletalMeshForPartOnPlayerPawn))
  (bind _returnvalue_15 (CallFunction|GetIdleVFXComponent))
  (bind _returnvalue_16 (CallFunction|GetIdleVFXComponent))
  (bind _returnvalue_17 (CallFunction|GetIdleVFXComponent))
  (bind _mesh_part (Variables|Default|GetMesh%20Part))
  (bind _returnvalue_18 (== _returnvalue_9 _mesh_part))
  (bind _mesh_part_2 (Variables|Default|GetMesh%20Part))
  (bind _mesh_part_3 (Variables|Default|GetMesh%20Part))
  (bind _mesh_part_4 (Variables|Default|GetMesh%20Part))
  (bind _bisinfrontendhologram (Variables|/Script/FortniteGame.FortPlayerPawn|GetbIsInFrontEndHologram _returnvalue_10))
  (bind _mesh_part_5 (Variables|Default|GetMesh%20Part))
  (bind _killcount (Variables|Default|GetKillCount))
  (bind _mesh_part_6 (Variables|Default|GetMesh%20Part))
  (bind _is_character (Variables|Default|GetIs%20Character))
  (|CallParentFunction)
  (if (CustomCosmeticModifier|IsinFrontEnd)
    (if (Variables|Default|GetIs%20Reactive)
      ;[0 0 0] FE Looping
      (Utilities|FlowControl|DoOnce false
        (:Completed
          (Timeline|StageChange PlayFromStart
            (:Stage1
              (if _bisinfrontendhologram
                _
                (elif _is_character
                  (Effects|Components|ParticleSystem|SetIntParameter _returnvalue_17 "FX Stage" 0)
                  (Rendering|Material|SetScalarParameterValueonMaterials _returnvalue_7 "FlameRevealAmount" 0.0)
                  (if (== (Customization|GetSkeletalMeshforPartType (CustomCharacter|GetAssociatedPlayerPawn) "Backpack" :!class /Script/FortniteGame.FortPlayerPawn) _returnvalue_12)
                    (Rendering|Material|SetScalarParameterValueonMaterials _returnvalue_12 "Emissive Mult" 9.0)
                    (else
                      (Rendering|Material|SetScalarParameterValueonMaterials _returnvalue_12 "Emissive Mult" 0.0)))
                  (else
                    (Effects|Components|ParticleSystem|SetIntParameter _returnvalue_17 "FX Stage" 0)
                    (Rendering|Material|SetScalarParameterValueonMaterials _returnvalue_7 "Backpack_Transition" 0.0)
                    (Rendering|Material|SetScalarParameterValueonMaterials _returnvalue_12 "Hide Element 03" 1.0)
                    (Rendering|Material|SetScalarParameterValueonMaterials _returnvalue_12 "Hide Element 02" 1.0)
                    (Rendering|Material|SetScalarParameterValueonMaterials _returnvalue_12 "Hide Element 01" 0.0)
                    (Rendering|Material|SetScalarParameterValueonMaterials _returnvalue_12 "Hide Element 04" 1.0)
                    (Rendering|Material|SetScalarParameterValueonMaterials _returnvalue_12 "Hide Element 03" 1.0)))))
            (:Staget2
              (if _bisinfrontendhologram
                _
                (elif _is_character
                  (Utilities|FlowControl|Sequence
                    (:then_0
                      (bind _returnvalue_19 (PartModifier|SpawnSystem|SpawnSystemAttachedPartModifier "/BRCosmetics/Characters/Player/Male/Medium/Bodies/M_MED_Headset_Fantastical/FX/NS_Headset_Fantastical_Tansform_FE.NS_Headset_Fantastical_Tansform_FE" _returnvalue_2 "shoulderPad_l_01")
                        (:Success
                          (Effects|Components|ParticleSystem|SetIntParameter _returnvalue_19 "FX Stage" 3)
                          (Effects|Components|ParticleSystem|SetIntParameter (CallFunction|GetIdleVFXComponent) "FX Stage" 3))))
                    (:then_1
                      _))
                  (Rendering|Material|SetScalarParameterValueonMaterials _returnvalue_5 "FlameRevealAmount" 0.6)
                  (if (not (== (Customization|GetSkeletalMeshforPartType (CustomCharacter|GetAssociatedPlayerPawn) "Backpack" :!class /Script/FortniteGame.FortPlayerPawn) _returnvalue_5))
                    (Rendering|Material|SetScalarParameterValueonMaterials _returnvalue_5 "Emissive Mult" 9.0))
                  (else
                    (bind _returnvalue_20 (PartModifier|SpawnSystem|SpawnSystemAttachedPartModifier "/BRCosmetics/Accessories/FORT_Backpacks/Backpack_M_MED_Headset/FX/NS_Backpack_Headset_Transform.NS_Backpack_Headset_Transform" _returnvalue_2 "root")
                      (:Success
                        (Effects|Components|ParticleSystem|SetIntParameter _returnvalue_20 "FX Stage" 2)
                        (Effects|Components|ParticleSystem|SetIntParameter (CallFunction|GetIdleVFXComponent) "FX Stage" 3)
                        (Rendering|Material|SetScalarParameterValueonMaterials _returnvalue_5 "Backpack_Transition" 0.5)
                        (Rendering|Material|SetScalarParameterValueonMaterials _returnvalue_3 "Hide Element 03" 0.0)
                        (Rendering|Material|SetScalarParameterValueonMaterials _returnvalue_3 "Hide Element 02" 0.0)
                        (Rendering|Material|SetScalarParameterValueonMaterials _returnvalue_3 "Hide Element 01" 1.0)))))))
            (:Staget3
              (if _bisinfrontendhologram
                _
                (elif _is_character
                  ;;[1 1 1] I moved this sooner because the next node fails on heads
                  (Rendering|Material|SetScalarParameterValueonMaterials _returnvalue_6 "FlameRevealAmount" 1.5)
                  (bind _returnvalue_21 (PartModifier|SpawnSystem|SpawnSystemAttachedPartModifier "/BRCosmetics/Characters/Player/Male/Medium/Bodies/M_MED_Headset_Fantastical/FX/NS_Headset_Fantastical_Tansform_FE.NS_Headset_Fantastical_Tansform_FE" _returnvalue "shoulderPad_l_01")
                    (:Success
                      (Effects|Components|ParticleSystem|SetBoolParameter _returnvalue_21 "bIsFrontEndPreview" true)
                      (Effects|Components|ParticleSystem|SetIntParameter _returnvalue_21 "FX Stage" 5)
                      (Effects|Components|ParticleSystem|SetIntParameter (CallFunction|GetIdleVFXComponent) "FX Stage" 5)
                      (Utilities|FlowControl|Delay 2.95)
                      (bind _returnvalue_22 (PartModifier|SpawnSystem|SpawnSystemAttachedPartModifier "/BRCosmetics/Characters/Player/Male/Medium/Bodies/M_MED_Headset_Fantastical/FX/NS_Headset_Fantastical_Tansform_FE.NS_Headset_Fantastical_Tansform_FE" (CustomCharacter|GetSkeletalMeshForPartOnPlayerPawn) "shoulderPad_l_01")
                        (:Success
                          (Effects|Components|ParticleSystem|SetIntParameter _returnvalue_22 "FX Stage" 5)))))
                  (else
                    (bind _returnvalue_23 (PartModifier|SpawnSystem|SpawnSystemAttachedPartModifier "/BRCosmetics/Accessories/FORT_Backpacks/Backpack_M_MED_Headset/FX/NS_Backpack_Headset_Transform.NS_Backpack_Headset_Transform" _returnvalue "root")
                      (:Success
                        (Effects|Components|ParticleSystem|SetIntParameter _returnvalue_23 "FX Stage" 3)
                        (Effects|Components|ParticleSystem|SetIntParameter _returnvalue_16 "FX Stage" 5)
                        (bind (_direction _grow_track_fe) (Timeline|Timeline_1 PlayFromStart)
                          (:Update
                            (Effects|Components|ParticleSystem|SetFloatParameter _returnvalue_16 "Grow" _grow_track_fe)
                            (Rendering|Material|SetScalarParameterValueonMaterials _returnvalue_6 "Backpack_Transition" 1.0)
                            (Rendering|Material|SetScalarParameterValueonMaterials _returnvalue_6 "FlameRevealAmount" 1.5)
                            (Rendering|Material|SetScalarParameterValueonMaterials _returnvalue_4 "Hide Element 04" 0.0)
                            (Rendering|Material|SetScalarParameterValueonMaterials _returnvalue_4 "Hide Element 03" 1.0)
                            (Utilities|FlowControl|Delay 2.95)
                            (bind _returnvalue_24 (PartModifier|SpawnSystem|SpawnSystemAttachedPartModifier "/BRCosmetics/Accessories/FORT_Backpacks/Backpack_M_MED_Headset/FX/NS_Backpack_Headset_Transform.NS_Backpack_Headset_Transform" (CustomCharacter|GetSkeletalMeshForPartOnPlayerPawn) "root")
                              (:Success
                                (Effects|Components|ParticleSystem|SetIntParameter _returnvalue_24 "FX Stage" 1))))))))))))))
      (elif (Variables|Default|GetIs%20Purple)
        (Effects|Components|ParticleSystem|SetIntParameter _returnvalue_15 "FX Stage" 5)
        (Rendering|Material|SetScalarParameterValueonMaterials _mesh_part_3 "FlameRevealAmount" 1.0)
        (else
          (Effects|Components|ParticleSystem|SetIntParameter _returnvalue_15 "FX Stage" 0)
          (Rendering|Material|SetScalarParameterValueonMaterials _mesh_part_3 "FlameRevealAmount" 0.25))))
    (else
      (bind (_statname _statvalue) (Cosmetics|WatchCosmeticStat (CustomCharacter|GetAssociatedPlayerPawn) "(TagName=\"StatManager.AthenaKills\")" 0 false)
        (:then
          (Development|PrintString "Watching Kills"))
        (:OnStatChanged
          (Utilities|FlowControl|Sequence
            (:then_0
              _)
            (:then_1
              (if (Variables|Default|GetIs%20Character)
                (if (Variables|Default|GetIs%20Reactive)
                  (if (Variables|Default|GetIs%20Character%20Body)
                    ;[0 0 0] Character Transform FX
                    (switch int (Variables|Default|GetKillCount)
                      (:0
                        ;[0 0 0] Backpack Transform FX
                        (Rendering|Material|SetScalarParameterValueonMaterials _mesh_part_5 "FlameRevealAmount" 0.25)
                        (Rendering|Material|SetScalarParameterValueonMaterials _mesh_part_5 "Emissive Mult" 0.0))
                      (:1
                        (bind _returnvalue_25 (Effects|SpawnFXSystemAttached "/BRCosmetics/Characters/Player/Male/Medium/Bodies/M_MED_Headset_Fantastical/FX/NS_Headset_Fantastical_Tansform.NS_Headset_Fantastical_Tansform" _returnvalue_11 "shoulderPad_l_01"))
                        (Effects|Components|ParticleSystem|SetIntParameter _returnvalue_25 "FX Stage" 1)
                        (Rendering|Material|SetScalarParameterValueonMaterials _mesh_part_5 "FlameRevealAmount" 0.39)
                        (Rendering|Material|SetScalarParameterValueonMaterials _returnvalue_11 "Emissive Mult" 9.0))
                      (:2
                        (bind _returnvalue_26 (Effects|SpawnFXSystemAttached "/BRCosmetics/Characters/Player/Male/Medium/Bodies/M_MED_Headset_Fantastical/FX/NS_Headset_Fantastical_Tansform.NS_Headset_Fantastical_Tansform" _returnvalue_11 "shoulderPad_l_01"))
                        (Effects|Components|ParticleSystem|SetIntParameter _returnvalue_26 "FX Stage" 2)
                        (Rendering|Material|SetScalarParameterValueonMaterials (Variables|Default|GetMesh%20Part) "FlameRevealAmount" 0.58))
                      (:3
                        (bind _returnvalue_27 (Effects|SpawnFXSystemAttached "/BRCosmetics/Characters/Player/Male/Medium/Bodies/M_MED_Headset_Fantastical/FX/NS_Headset_Fantastical_Tansform.NS_Headset_Fantastical_Tansform" _returnvalue_11 "shoulderPad_l_01"))
                        (Effects|Components|ParticleSystem|SetIntParameter _returnvalue_27 "FX Stage" 2)
                        (Rendering|Material|SetScalarParameterValueonMaterials (Variables|Default|GetMesh%20Part) "FlameRevealAmount" 0.72))
                      (:4
                        (bind _returnvalue_28 (Effects|SpawnFXSystemAttached "/BRCosmetics/Characters/Player/Male/Medium/Bodies/M_MED_Headset_Fantastical/FX/NS_Headset_Fantastical_Tansform.NS_Headset_Fantastical_Tansform" _returnvalue_11 "shoulderPad_l_01"))
                        (Effects|Components|ParticleSystem|SetIntParameter _returnvalue_28 "FX Stage" 3)
                        (Rendering|Material|SetScalarParameterValueonMaterials (Variables|Default|GetMesh%20Part) "FlameRevealAmount" 0.82))
                      (:5
                        (bind _returnvalue_29 (Effects|SpawnFXSystemAttached "/BRCosmetics/Characters/Player/Male/Medium/Bodies/M_MED_Headset_Fantastical/FX/NS_Headset_Fantastical_Tansform.NS_Headset_Fantastical_Tansform" _returnvalue_11 "shoulderPad_l_01"))
                        (Effects|Components|ParticleSystem|SetIntParameter _returnvalue_29 "FX Stage" 4)
                        (Rendering|Material|SetScalarParameterValueonMaterials (Variables|Default|GetMesh%20Part) "FlameRevealAmount" 1.25)))
                    (else
                      (switch int (Variables|Default|GetKillCount)
                        (:0
                          (Rendering|Material|SetScalarParameterValueonMaterials _mesh_part_6 "FlameRevealAmount" 0.25)
                          (Rendering|Material|SetScalarParameterValueonMaterials _mesh_part_6 "emissive mult" 0.0)
                          ;[1 1 1] Dont hide elements on face
                          (bind _returnvalue_30 (CallFunction|FaceAccessoryCheck))
                          (if (not _returnvalue_30)
                            (Rendering|Material|SetScalarParameterValueonMaterials _returnvalue_13 "Hide Element 01" 0.0)))
                        (:1
                          (Rendering|Material|SetScalarParameterValueonMaterials (Variables|Default|GetMesh%20Part) "FlameRevealAmount" 0.39)
                          (Rendering|Material|SetScalarParameterValueonMaterials (CustomCharacter|GetSkeletalMeshForPartOnPlayerPawn) "emissive mult" 9.0)
                          (bind _returnvalue_31 (CallFunction|FaceAccessoryCheck))
                          (if (not _returnvalue_31)
                            (Rendering|Material|SetScalarParameterValueonMaterials _returnvalue_13 "Hide Element 01" 1.0)))
                        (:2
                          (Rendering|Material|SetScalarParameterValueonMaterials (Variables|Default|GetMesh%20Part) "FlameRevealAmount" 0.58))
                        (:3
                          (Rendering|Material|SetScalarParameterValueonMaterials (Variables|Default|GetMesh%20Part) "FlameRevealAmount" 0.72))
                        (:4
                          (Rendering|Material|SetScalarParameterValueonMaterials (Variables|Default|GetMesh%20Part) "FlameRevealAmount" 0.82))
                        (:5
                          (Rendering|Material|SetScalarParameterValueonMaterials (Variables|Default|GetMesh%20Part) "FlameRevealAmount" 1.25)))))
                  (else
                    (Development|PrintString "Hello")
                    (Rendering|Material|SetScalarParameterValueonMaterials _mesh_part_4 "FlameRevealAmount" 0.25)
                    (Rendering|Material|SetScalarParameterValueonMaterials _mesh_part_4 "Emissive Mult" 9.0)))
                (elif (Variables|Default|GetIs%20Reactive)
                  (Utilities|FlowControl|Sequence
                    (:then_0
                      ;[0 0 0] Backpack Kill FX
                      (if (> (Math|Conversions|ToFloat(Integer) _killcount) 0.0)
                        (bind _returnvalue_32 (Effects|SpawnFXSystemAttached "/BRCosmetics/Accessories/FORT_Backpacks/Backpack_M_MED_Headset/FX/NS_Backpack_Headset_Kill.NS_Backpack_Headset_Kill" (CustomCharacter|GetSkeletalMeshForPartOnPlayerPawn) "root"))
                        (switch int _killcount
                          (:Default
                            _)
                          (:0
                            (Effects|Components|ParticleSystem|SetIntParameter _returnvalue_32 "FX Stage" 0)
                            (Rendering|Material|SetScalarParameterValueonMaterials (Variables|Default|GetMesh%20Part) "Backpack_Transition" 0.0))
                          (:1
                            (Effects|Components|ParticleSystem|SetIntParameter _returnvalue_32 "FX Stage" 1)
                            (Rendering|Material|SetScalarParameterValueonMaterials (Variables|Default|GetMesh%20Part) "Backpack_Transition" 0.0)
                            ;[1 1 1] If Backpack Hide/show by vert color
                            (if _returnvalue_18
                              (Rendering|Material|SetScalarParameterValueonMaterials _mesh_part "Hide Element 03" 0.0)
                              (Rendering|Material|SetScalarParameterValueonMaterials _mesh_part "Hide Element 02" 0.0)
                              (Rendering|Material|SetScalarParameterValueonMaterials _mesh_part "Hide Element 01" 1.0)))
                          (:2
                            (Effects|Components|ParticleSystem|SetIntParameter _returnvalue_32 "FX Stage" 2)
                            (Rendering|Material|SetScalarParameterValueonMaterials (Variables|Default|GetMesh%20Part) "Backpack_Transition" 0.5))
                          (:3
                            (Effects|Components|ParticleSystem|SetIntParameter _returnvalue_32 "FX Stage" 2)
                            (Rendering|Material|SetScalarParameterValueonMaterials _mesh_part_2 "Backpack_Transition" 0.5))
                          (:4
                            (Effects|Components|ParticleSystem|SetIntParameter _returnvalue_32 "FX Stage" 2))
                          (:5
                            _))
                        (Effects|Components|ParticleSystem|SetIntParameter _returnvalue_32 "FX Stage" 3)
                        (Rendering|Material|SetScalarParameterValueonMaterials _mesh_part_2 "Backpack_Transition" 1.0)
                        (if _returnvalue_18
                          (Rendering|Material|SetScalarParameterValueonMaterials _mesh_part "Hide Element 04" 0.0)
                          (Rendering|Material|SetScalarParameterValueonMaterials _mesh_part "Hide Element 03" 1.0))))
                    (:then_1
                      (switch int (Variables|Default|GetKillCount)
                        (:2
                          (bind _returnvalue_33 (Effects|SpawnFXSystemAttached "/BRCosmetics/Accessories/FORT_Backpacks/Backpack_M_MED_Headset/FX/NS_Backpack_Headset_Transform.NS_Backpack_Headset_Transform" _returnvalue_14 "root"))
                          (Effects|Components|ParticleSystem|SetIntParameter _returnvalue_33 "FX Stage" 2))
                        (:5
                          (bind _returnvalue_34 (Effects|SpawnFXSystemAttached "/BRCosmetics/Accessories/FORT_Backpacks/Backpack_M_MED_Headset/FX/NS_Backpack_Headset_Transform.NS_Backpack_Headset_Transform" _returnvalue_14 "root"))
                          (Effects|Components|ParticleSystem|SetIntParameter _returnvalue_34 "FX Stage" 3)))))))))))
      (Variables|Default|SetKillCount _statvalue)
      (bind _delegatehandlecontroller_out (Cosmetics|WhenCosmeticPartApplied (CustomCharacter|GetAssociatedPlayerPawn) (delegate Custom|OnCosmeticPartApplied) true))
      (Variables|Default|SetDelegate%20Handle%20Controller _delegatehandlecontroller_out))))

(event Custom|OnCosmeticPartApplied
  (params (Pawn :type Fort%20Player%20Pawn%20Object%20Reference) (Part :type Custom%20Character%20Part%20Object%20Reference) (MeshPart :type Skeletal%20Mesh%20Component%20Object%20Reference) (PartType :type EFortCustomPartType%20Enum :default "Head"))
  (bind _returnvalue (CustomCharacter|GetAssociatedPlayerPawn))
  (bind _returnvalue_2 (Customization|GetSkeletalMeshforPartType _returnvalue "Backpack" :!class /Script/FortniteGame.FortPlayerPawn))
  (bind _returnvalue_3 (CustomCharacter|GetAssociatedPlayerPawn))
  (bind _returnvalue_4 (Customization|GetSkeletalMeshforPartType _returnvalue_3 "Backpack" :!class /Script/FortniteGame.FortPlayerPawn))
  (bind _returnvalue_5 (CallFunction|GetIdleVFXComponent))
  (bind _mesh_part (Variables|Default|GetMesh%20Part))
  (bind _returnvalue_6 (== _returnvalue_4 _mesh_part))
  (bind _mesh_part_2 (Variables|Default|GetMesh%20Part))
  (bind _mesh_part_3 (Variables|Default|GetMesh%20Part))
  (if (Utilities|Enum|Equal(Enum) PartType (CustomCharacter|GetCharacterPartType))
    (Variables|Default|SetMesh%20Part MeshPart)
    ;[0 0 0] Idle FX
    (if (Variables|Default|GetIs%20Reactive)
      (switch int (Variables|Default|GetKillCount)
        (:0
          (Effects|Components|ParticleSystem|SetIntParameter _returnvalue_5 "FX Stage" 0)
          (Rendering|Material|SetScalarParameterValueonMaterials _mesh_part_3 "Backpack_Transition" 0.0)
          (if (== _returnvalue_2 _mesh_part_3)
            (if (== _returnvalue_2 "/Game/Accessories/FORT_Backpacks/Backpack_M_MED_Headset/Meshes/M_MED_Headset_Pack.M_MED_Headset_Pack")
              (Rendering|Material|SetScalarParameterValueonMaterials _mesh_part_3 "Emissive Mult" 9.0))
            (else
              (Rendering|Material|SetScalarParameterValueonMaterials _mesh_part_3 "Emissive Mult" 0.0))))
        (:1
          (Effects|Components|ParticleSystem|SetIntParameter _returnvalue_5 "FX Stage" 1)
          (Rendering|Material|SetScalarParameterValueonMaterials (Variables|Default|GetMesh%20Part) "Backpack_Transition" 0.0)
          ;[1 1 1] if Backpack hide by vert color
          (if _returnvalue_6
            (Rendering|Material|SetScalarParameterValueonMaterials _mesh_part "Hide Element 03" 0.0)
            (Rendering|Material|SetScalarParameterValueonMaterials _mesh_part "Hide Element 02" 0.0)
            (Rendering|Material|SetScalarParameterValueonMaterials _mesh_part "Hide Element 01" 1.0)))
        (:2
          (Effects|Components|ParticleSystem|SetIntParameter _returnvalue_5 "FX Stage" 2)
          (Rendering|Material|SetScalarParameterValueonMaterials (Variables|Default|GetMesh%20Part) "Backpack_Transition" 0.5))
        (:3
          (Effects|Components|ParticleSystem|SetIntParameter _returnvalue_5 "FX Stage" 3)
          (Rendering|Material|SetScalarParameterValueonMaterials _mesh_part "Backpack_Transition" 0.5))
        (:4
          (Effects|Components|ParticleSystem|SetIntParameter _returnvalue_5 "FX Stage" 4))
        (:5
          (Effects|Components|ParticleSystem|SetIntParameter _returnvalue_5 "FX Stage" 5)
          (Utilities|FlowControl|Sequence
            (:then_0
              (Rendering|Material|SetScalarParameterValueonMaterials (Variables|Default|GetMesh%20Part) "Backpack_Transition" 1.0)
              (if _returnvalue_6
                (Rendering|Material|SetScalarParameterValueonMaterials _mesh_part "Hide Element 04" 0.0)
                (Rendering|Material|SetScalarParameterValueonMaterials _mesh_part "Hide Element 03" 1.0)))
            (:then_1
              (bind (_direction _grow_track) (Timeline|Timeline Play)
                (:Update
                  (Effects|Components|ParticleSystem|SetFloatParameter _returnvalue_5 "Grow" _grow_track)))))))
      (elif (Variables|Default|GetIs%20Purple)
        (Effects|Components|ParticleSystem|SetIntParameter _returnvalue_5 "FX Stage" 5)
        (Rendering|Material|SetScalarParameterValueonMaterials _mesh_part_2 "FlameRevealAmount" 1.0)
        (else
          (Effects|Components|ParticleSystem|SetIntParameter _returnvalue_5 "FX Stage" 0)
          (Rendering|Material|SetScalarParameterValueonMaterials _mesh_part_2 "FlameRevealAmount" 0.0))))))

(fn FaceAccessoryCheck
  (return (== (Customization|GetSkeletalMeshforPartType (CustomCharacter|GetAssociatedPlayerPawn) "Face" :!class /Script/FortniteGame.FortPlayerPawn) (Variables|Default|GetMesh%20Part))))

(fn ConstructionScript
  (|CallParentFunction))
