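; Part modifier for the Headset backpack cosmetic, evolving its materials and particles with the player's kills when reactive.
; In match, it watches the AthenaKills stat to spawn kill bursts, update staged material and idle FX parameters, and fire transform effects at two and five kills.
; A timeline grows the final idle effect, skipped if already earned.
; Front end previews stages via timeline; non-reactive variants apply static purple or default looks.
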(event Custom|CE_Frontend%20FX%20Loop%20Burst)

(event EventOnPawnAndPartChanged
  (bind _returnvalue (CustomCharacter|GetSkeletalMeshForPartOnPlayerPawn))
  (bind _returnvalue_2 (CustomCharacter|GetSkeletalMeshForPartOnPlayerPawn))
  (bind _returnvalue_3 (CustomCharacter|GetSkeletalMeshForPartOnPlayerPawn))
  (bind _returnvalue_4 (CustomCharacter|GetAssociatedPlayerPawn))
  (bind _returnvalue_5 (CustomCharacter|GetSkeletalMeshForPartOnPlayerPawn))
  (bind _returnvalue_6 (CustomCharacter|GetAssociatedPlayerPawn))
  (bind _returnvalue_7 (CallFunction|GetIdleVFXComponent))
  (bind _returnvalue_8 (CallFunction|GetIdleVFXComponent))
  (bind _mesh_part (Variables|Default|GetMesh%20Part))
  (bind _bisinfrontendhologram (Variables|/Script/FortniteGame.FortPlayerPawn|GetbIsInFrontEndHologram _returnvalue_4))
  (bind _killcount (Variables|Default|GetKillCount))
  (|CallParentFunction)
  (if (CustomCosmeticModifier|IsinFrontEnd)
    (if (Variables|Default|GetIs%20Reactive)
      ;[0 0 0] FE Looping
      (Utilities|FlowControl|DoOnce false
        (:Completed
          (Timeline|StageChange PlayFromStart
            (:Stage1
              (if (not _bisinfrontendhologram)
                (Effects|Components|ParticleSystem|SetIntParameter (CallFunction|GetIdleVFXComponent) "FX Stage" 0)
                (Rendering|Material|SetScalarParameterValueonMaterials (CustomCharacter|GetSkeletalMeshForPartOnPlayerPawn) "Backpack_Transition" 0.0)
                (Rendering|Material|SetScalarParameterValueonMaterials _returnvalue_5 "Hide Element 03" 1.0)
                (Rendering|Material|SetScalarParameterValueonMaterials _returnvalue_5 "Hide Element 02" 1.0)
                (Rendering|Material|SetScalarParameterValueonMaterials _returnvalue_5 "Hide Element 01" 0.0)
                (Rendering|Material|SetScalarParameterValueonMaterials _returnvalue_5 "Hide Element 04" 1.0)
                (Rendering|Material|SetScalarParameterValueonMaterials _returnvalue_5 "Hide Element 03" 1.0)))
            (:Staget2
              (if (not _bisinfrontendhologram)
                (bind _returnvalue_9 (PartModifier|SpawnSystem|SpawnSystemAttachedPartModifier "/BRCosmetics/Accessories/FORT_Backpacks/Backpack_M_MED_Headset/FX/NS_Backpack_Headset_Transform.NS_Backpack_Headset_Transform" (CustomCharacter|GetSkeletalMeshForPartOnPlayerPawn) "root")
                  (:Success
                    (Effects|Components|ParticleSystem|SetIntParameter _returnvalue_9 "FX Stage" 2)
                    (Effects|Components|ParticleSystem|SetIntParameter (CallFunction|GetIdleVFXComponent) "FX Stage" 3)
                    (Rendering|Material|SetScalarParameterValueonMaterials (CustomCharacter|GetSkeletalMeshForPartOnPlayerPawn) "Backpack_Transition" 0.5)
                    (Rendering|Material|SetScalarParameterValueonMaterials _returnvalue_2 "Hide Element 03" 0.0)
                    (Rendering|Material|SetScalarParameterValueonMaterials _returnvalue_2 "Hide Element 02" 0.0)
                    (Rendering|Material|SetScalarParameterValueonMaterials _returnvalue_2 "Hide Element 01" 1.0)))))
            (:Staget3
              (if (not _bisinfrontendhologram)
                (bind _returnvalue_10 (PartModifier|SpawnSystem|SpawnSystemAttachedPartModifier "/BRCosmetics/Accessories/FORT_Backpacks/Backpack_M_MED_Headset/FX/NS_Backpack_Headset_Transform.NS_Backpack_Headset_Transform" (CustomCharacter|GetSkeletalMeshForPartOnPlayerPawn) "root")
                  (:Success
                    (Effects|Components|ParticleSystem|SetIntParameter _returnvalue_10 "FX Stage" 3)
                    (Effects|Components|ParticleSystem|SetIntParameter _returnvalue_8 "FX Stage" 5)
                    (bind (_direction _grow_track_fe) (Timeline|Timeline PlayFromStart)
                      (:Update
                        (Effects|Components|ParticleSystem|SetFloatParameter _returnvalue_8 "Grow" _grow_track_fe)
                        (Rendering|Material|SetScalarParameterValueonMaterials _returnvalue "Backpack_Transition" 1.0)
                        (Rendering|Material|SetScalarParameterValueonMaterials _returnvalue "FlameRevealAmount" 1.5)
                        (Rendering|Material|SetScalarParameterValueonMaterials _returnvalue_3 "Hide Element 04" 0.0)
                        (Rendering|Material|SetScalarParameterValueonMaterials _returnvalue_3 "Hide Element 03" 1.0)
                        (Utilities|FlowControl|Delay 2.95)
                        (bind _returnvalue_11 (PartModifier|SpawnSystem|SpawnSystemAttachedPartModifier "/BRCosmetics/Accessories/FORT_Backpacks/Backpack_M_MED_Headset/FX/NS_Backpack_Headset_Transform.NS_Backpack_Headset_Transform" (CustomCharacter|GetSkeletalMeshForPartOnPlayerPawn) "root")
                          (:Success
                            (Effects|Components|ParticleSystem|SetIntParameter _returnvalue_11 "FX Stage" 1))))))))))))
      (elif (Variables|Default|GetIs%20Purple)
        (Effects|Components|ParticleSystem|SetIntParameter _returnvalue_7 "FX Stage" 5)
        (Rendering|Material|SetScalarParameterValueonMaterials _mesh_part "FlameRevealAmount" 1.0)
        (else
          (Effects|Components|ParticleSystem|SetIntParameter _returnvalue_7 "FX Stage" 0)
          (Rendering|Material|SetScalarParameterValueonMaterials _mesh_part "FlameRevealAmount" 0.25))))
    (elif (Variables|Default|GetIs%20Reactive)
      (bind _returnvalue_12 (Class|FortCosmeticFunctionLibrary|GetCosmeticStatValue _returnvalue_6 "(TagName=\"StatManager.AthenaKills\")"))
      (Variables|Default|SetKillCount _returnvalue_12)
      (Utilities|FlowControl|Sequence
        (:then_0
          ;[1 1 1] If our kill count was >= 5 when we spawned, act is if the grow animation completed
          (if (>= (Variables|Default|GetKillCount) 5)
            (Variables|Default|SetIdle%20FX%20Grow 1.0)))
        (:then_1
          (CallFunction|PushMaterialParameters)
          (CallFunction|UpdateIdleFX)
          (bind (_statname _statvalue) (Cosmetics|WatchCosmeticStat _returnvalue_6 "(TagName=\"StatManager.AthenaKills\")" 0 false)
            (:OnStatChanged
              (Utilities|FlowControl|Sequence
                (:then_0
                  (Variables|Default|SetKillCount _statvalue))
                (:then_1
                  (if (> _killcount 0)
                    (CallFunction|SpawnKillFX _killcount)))
                (:then_2
                  (CallFunction|PushMaterialParameters))
                (:then_3
                  (switch int (Variables|Default|GetKillCount)
                    (:2
                      (CallFunction|SpawnTransformFX 2))
                    (:5
                      (CallFunction|SpawnTransformFX 3))))
                (:then_4
                  (CallFunction|UpdateIdleFX)
                  (if (and (>= (Variables|Default|GetKillCount) 5) (Math|Float|NearlyEqual(Float) (Variables|Default|GetIdle%20FX%20Grow)))
                    (bind (_direction_2 _grow_track) (Timeline|Timeline_0 Play)
                      (:Update
                        (bind _output_get (Variables|Default|SetIdle%20FX%20Grow _grow_track))
                        (Effects|Components|ParticleSystem|SetFloatParameter (CallFunction|GetIdleVFXComponent) "Grow" _output_get)))))))))))))

(event EventOnPostResIn
  (bind _returnvalue (CallFunction|GetIdleVFXComponent))
  (bind _mesh_part (Variables|Default|GetMesh%20Part))
  (|CallParentFunction)
  (if (Variables|Default|GetIs%20Reactive)
    (Utilities|FlowControl|Sequence
      (:then_0
        (CallFunction|PushMaterialParameters))
      (:then_1
        (CallFunction|UpdateIdleFX)))
    (elif (Variables|Default|GetIs%20Purple)
      (Effects|Components|ParticleSystem|SetIntParameter _returnvalue "FX Stage" 5)
      (Rendering|Material|SetScalarParameterValueonMaterials _mesh_part "FlameRevealAmount" 1.0)
      (else
        (Effects|Components|ParticleSystem|SetIntParameter _returnvalue "FX Stage" 0)
        (Rendering|Material|SetScalarParameterValueonMaterials _mesh_part "FlameRevealAmount" 0.0)))))

(fn CalculateKillFXStage
  (params (Kill%20Count :type Integer))
  (switch int Kill%20Count
    (:Default
      _)
    (:0
      (return 0))
    (:1
      (return 1))
    (:2
      (return 2))
    (:3
      (return 2))
    (:4
      (return 2))
    (:5
      _))
  (return 3))

(fn CalculateBackpackTransition
  (params (Kill%20Count :type Integer))
  (switch int Kill%20Count
    (:Default
      _)
    (:0
      (return 0.0))
    (:1
      (return 0.0))
    (:2
      (return 0.5))
    (:3
      (return 0.5))
    (:4
      _)
    (:5
      _))
  (return 1.0))

(fn FaceAccessoryCheck
  (return false))

(fn PushMaterialParameters
  (bind _part_mesh_component (LocalVariables|GetPart%20Mesh%20Component))
  (bind _part_mesh_component_2 (LocalVariables|GetPart%20Mesh%20Component))
  (bind _part_mesh_component_3 (LocalVariables|GetPart%20Mesh%20Component))
  (bind _output_get (LocalVariables|SetPart%20Mesh%20Component (CustomCharacter|GetSkeletalMeshForPartOnPlayerPawn)))
  (Utilities|IsValid _output_get
    (:Is%20Valid
      (Utilities|FlowControl|Sequence
        (:then_0
          (Rendering|Material|SetScalarParameterValueonMaterials (LocalVariables|GetPart%20Mesh%20Component) "Backpack_Transition" (CallFunction|CalculateBackpackTransition (Variables|Default|GetKillCount))))
        (:then_1
          (switch int (Variables|Default|GetKillCount)
            (:Default
              _)
            (:0
              (Rendering|Material|SetScalarParameterValueonMaterials _part_mesh_component_2 "Hide Element 01" 0.0)
              (Rendering|Material|SetScalarParameterValueonMaterials _part_mesh_component_2 "Hide Element 02" 1.0)
              (Rendering|Material|SetScalarParameterValueonMaterials _part_mesh_component_2 "Hide Element 03" 1.0)
              (Rendering|Material|SetScalarParameterValueonMaterials _part_mesh_component_2 "Hide Element 04" 1.0))
            (:1
              (Rendering|Material|SetScalarParameterValueonMaterials _part_mesh_component_3 "Hide Element 01" 1.0 :!label _label)
              (Rendering|Material|SetScalarParameterValueonMaterials _part_mesh_component_3 "Hide Element 02" 0.0)
              (Rendering|Material|SetScalarParameterValueonMaterials _part_mesh_component_3 "Hide Element 03" 0.0)
              (Rendering|Material|SetScalarParameterValueonMaterials _part_mesh_component_3 "Hide Element 04" 1.0))
            (:2
              (goto _label))
            (:3
              (goto _label))
            (:4
              (goto _label))
            (:5
              _))
          (Rendering|Material|SetScalarParameterValueonMaterials _part_mesh_component "Hide Element 01" 1.0)
          (Rendering|Material|SetScalarParameterValueonMaterials _part_mesh_component "Hide Element 02" 0.0)
          (Rendering|Material|SetScalarParameterValueonMaterials _part_mesh_component "Hide Element 03" 1.0)
          (Rendering|Material|SetScalarParameterValueonMaterials _part_mesh_component "Hide Element 04" 0.0))))))

(fn SpawnKillFX
  (params (Kill%20Count :type Integer))
  (bind _returnvalue (PartModifier|SpawnSystem|SpawnSystemAttachedPartModifier "/BRCosmetics/Accessories/FORT_Backpacks/Backpack_M_MED_Headset/FX/NS_Backpack_Headset_Kill.NS_Backpack_Headset_Kill" (CustomCharacter|GetSkeletalMeshForPartOnPlayerPawn) "root")
    (:Success
      (Effects|Components|ParticleSystem|SetIntParameter _returnvalue "FX Stage" (CallFunction|CalculateKillFXStage (Variables|GetKill%20Count))))))

(fn SpawnTransformFX
  (params (FX%20Stage :type Integer))
  (bind _returnvalue (PartModifier|SpawnSystem|SpawnSystemAttachedPartModifier "/BRCosmetics/Accessories/FORT_Backpacks/Backpack_M_MED_Headset/FX/NS_Backpack_Headset_Transform.NS_Backpack_Headset_Transform" (CustomCharacter|GetSkeletalMeshForPartOnPlayerPawn) "root")
    (:Success
      (Effects|Components|ParticleSystem|SetIntParameter _returnvalue "FX Stage" (Variables|GetFX%20Stage)))))

(fn UpdateIdleFX
  (bind _returnvalue (CallFunction|GetIdleVFXComponent))
  (Utilities|IsValid _returnvalue
    (:Is%20Valid
      (Effects|Components|ParticleSystem|SetIntParameter _returnvalue "FX Stage" (Math|Integer|Min(Integer) (Variables|Default|GetKillCount) 5))
      (Effects|Components|ParticleSystem|SetFloatParameter _returnvalue "Grow" (Variables|Default|GetIdle%20FX%20Grow)))))

(fn ConstructionScript
  (|CallParentFunction))
