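; Cosmetic part modifier for the Headset back bling that makes its effects and materials react to the player's kills.
; In the front end, non-reactive parts get the idle VFX stage and flame reveal directly.
; In matches it watches the AthenaKills stat, spawning attached Niagara kill FX and changing material parameters, including element visibility on the Backpack mesh.
; A part-applied delegate caches the mesh and sets the same idle VFX and flame reveal.
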(event EventOnPawnAndPartChanged
  (bind _mesh_part (Variables|Default|GetMesh%20Part))
  (|CallParentFunction)
  (if (CustomCosmeticModifier|IsinFrontEnd)
    (if (not (Variables|Default|GetIs%20Reactive))
      (Effects|Components|ParticleSystem|SetIntParameter (CallFunction|GetIdleVFXComponent) "FX Stage" 5)
      (Rendering|Material|SetScalarParameterValueonMaterials (Variables|Default|GetMesh%20Part) "FlameRevealAmount" 1.0))
    (else
      (bind (_statname _statvalue) (Cosmetics|WatchCosmeticStat (CustomCharacter|GetAssociatedPlayerPawn) "(TagName=\"StatManager.AthenaKills\")")
        (:then
          (Development|PrintString "Watching Kills"))
        (:OnStatChanged
          (Utilities|FlowControl|Sequence
            (:then_0
              _)
            (:then_1
              ;[0 0 0] Backpack Kill FX
              (if (> (Math|Conversions|ToFloat(Integer) (Variables|Default|GetKillCount)) 0.0)
                (bind _returnvalue (Effects|SpawnFXSystemAttached "/BRCosmetics/Accessories/FORT_Backpacks/Backpack_M_MED_Headset/FX/NS_Backpack_Headset_Kill.NS_Backpack_Headset_Kill" (CustomCharacter|GetSkeletalMeshForPartOnPlayerPawn) "root"))
                (Effects|Components|ParticleSystem|SetIntParameter _returnvalue "FX Stage" 3)
                (Rendering|Material|SetScalarParameterValueonMaterials (Variables|Default|GetMesh%20Part) "Backpack_Transition" 1.0)
                ;[1 1 1] If Backpack Hide/show by vert color
                (if (== (Customization|GetSkeletalMeshforPartType (CustomCharacter|GetAssociatedPlayerPawn) "Backpack" :!class /Script/FortniteGame.FortPlayerPawn) _mesh_part)
                  (Rendering|Material|SetScalarParameterValueonMaterials _mesh_part "Hide Element 04" 0.0)
                  (Rendering|Material|SetScalarParameterValueonMaterials _mesh_part "Hide Element 03" 1.0)))))))
      (Variables|Default|SetKillCount _statvalue)
      (bind _delegatehandlecontroller_out (Cosmetics|WhenCosmeticPartApplied (CustomCharacter|GetAssociatedPlayerPawn) (delegate Custom|OnCosmeticPartApplied) true))
      (Variables|Default|SetDelegate%20Handle%20Controller _delegatehandlecontroller_out))))

(event Custom|OnCosmeticPartApplied
  (params (Pawn :type Fort%20Player%20Pawn%20Object%20Reference) (Part :type Custom%20Character%20Part%20Object%20Reference) (MeshPart :type Skeletal%20Mesh%20Component%20Object%20Reference) (PartType :type EFortCustomPartType%20Enum :default "Head"))
  (Variables|Default|SetMesh%20Part MeshPart)
  (Effects|Components|ParticleSystem|SetIntParameter (CallFunction|GetIdleVFXComponent) "FX Stage" 5)
  (Rendering|Material|SetScalarParameterValueonMaterials (Variables|Default|GetMesh%20Part) "FlameRevealAmount" 1.0))

(fn FaceAccessoryCheck
  (return (== (Customization|GetSkeletalMeshforPartType (CustomCharacter|GetAssociatedPlayerPawn) "Face" :!class /Script/FortniteGame.FortPlayerPawn) (Variables|Default|GetMesh%20Part))))

(fn ConstructionScript
  (|CallParentFunction))
