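; Character part modifier for the HeavyRoar cosmetic that makes the outfit react to the player's kills.
; On post-init it caches the part's MID and base values, detects the Mat3 style variant and watches the kill stat.
; Each kill plays a timeline boosting emissive and gradient brightness, scaled by day phase; the variant also triggers Niagara effects.
; In frontend contexts, a looping timer previews the reaction.
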(event Custom|FrontEnd_ReactivityPreview
  (bind _returnvalue (TimeOfDay|GetCurrentDayPhase :!class /Script/FortniteGame.FortKismetLibrary))
  (bind _boostfactor (Variables|Default|GetBoostFactor))
  (bind _partmid (Variables|Default|GetPartMID))
  (bind _boostfactortod (Variables|Default|GetBoostFactorTOD))
  ;[1 1 1] Displays effect on frontend
  (if (Reactive|Frontend|IsCosmeticObjectDisplayedinFrontendContext 0)
    (Utilities|FlowControl|Sequence
      (:then_0
        (if (Utilities|Enum|Equal(Enum) _returnvalue "Day")
          (Variables|Default|SetBoostFactorTOD _boostfactor)
          (bind (_direction _boost) (Timeline|Boost PlayFromStart)
            (:Update
              (Rendering|Material|SetScalarParameterValue _partmid "emissive mult" (* (+ (* _boost _boostfactortod) 1.0) (Variables|Default|GetInitialEmissiveIntensity)) :!class /Script/Engine.MaterialInstanceDynamic)
              (Rendering|Material|SetScalarParameterValue _partmid "UV Gradient_Brightness" (* (+ (* _boost _boostfactortod) 1.0) (Variables|Default|GetInitial%20UV%20Gradient_Brightness)) :!class /Script/Engine.MaterialInstanceDynamic)))
          (else
            (Variables|Default|SetBoostFactorTOD (Math|Float|SelectFloat (* _boostfactor (Variables|Default|GetBoostFactorNightFactor)) (* (Variables|Default|GetBoostFactor) (Variables|Default|GetBoostFactorMorningEveningFactor)) (Utilities|Enum|Equal(Enum) _returnvalue "Night")))
            (Timeline|Boost PlayFromStart))))
      (:then_1
        (if (Variables|Default|GetStyleVar)
          (CallFunction|FXonKillStatChange))))
    (else
      (Utilities|Time|ClearandInvalidateTimerbyHandle (Variables|Default|GetFrontEndReactivityTimer)))))

(event EventOnPostResIn
  (bind _returnvalue (TimeOfDay|GetCurrentDayPhase :!class /Script/FortniteGame.FortKismetLibrary))
  (bind _partmid (Variables|Default|GetPartMID))
  (bind _boostfactor (Variables|Default|GetBoostFactor))
  (bind _partmid_2 (Variables|Default|GetPartMID))
  (bind _boostfactortod (Variables|Default|GetBoostFactorTOD))
  (|CallParentFunction)
  (Utilities|FlowControl|Sequence
    (:then_0
      ;[1 0.08089 0] is "Style" Variant
      (bind (_index _bindexisvalid) (Class|FortCosmeticFunctionLibrary|GetVariantChannelActiveIndexForSlot "(TagName=\"LoadoutSlot.Character\")" (CosmeticLoadout|GetCosmeticLoadout) "(TagName=\"Cosmetics.Variant.Channel.Material\")" (Utilities|Array|MakeArray "(TagName=\"Cosmetics.Variant.Property.Mat3\")") false false))
      (Variables|Default|SetStyleVar _bindexisvalid))
    (:then_1
      (bind _asmaterial_instance_dynamic (Utilities|Casting|CastToMaterialInstanceDynamic (Rendering|Material|GetMaterial (CustomCharacter|GetSkeletalMeshForPartOnPlayerPawn) 0)))
      (bind _output_get (Variables|Default|SetPartMID _asmaterial_instance_dynamic))
      (Utilities|IsValid _output_get
        (:Is%20Valid
          (Utilities|FlowControl|Sequence
            (:then_0
              (bind _returnvalue_2 (Rendering|Material|GetScalarParameterValue _partmid "UV Gradient_Brightness" :!class /Script/Engine.MaterialInstanceDynamic))
              (Variables|Default|SetInitial%20UV%20Gradient_Brightness _returnvalue_2))
            (:then_1
              (bind _returnvalue_3 (Rendering|Material|GetScalarParameterValue _partmid "emissive mult" :!class /Script/Engine.MaterialInstanceDynamic))
              (Variables|Default|SetInitialEmissiveIntensity _returnvalue_3))
            (:then_2
              (Variables|Default|SetAssociatedPlayerPawn (CustomCharacter|GetAssociatedPlayerPawn))
              (Cosmetics|WatchCosmeticStat (Variables|Default|GetAssociatedPlayerPawn) "(TagName=\"StatManager.AthenaKills\")" -1 false
                (:OnStatChanged
                  ;[1 0.4851 0] Req : jira FORT-574107
                  (Utilities|FlowControl|Delay 0.5)
                  (Utilities|FlowControl|Sequence
                    (:then_0
                      (if (Utilities|Enum|Equal(Enum) _returnvalue "Day")
                        (Variables|Default|SetBoostFactorTOD _boostfactor)
                        (Timeline|Boost PlayFromStart)
                        (else
                          (Variables|Default|SetBoostFactorTOD (Math|Float|SelectFloat (* _boostfactor (Variables|Default|GetBoostFactorNightFactor)) (* (Variables|Default|GetBoostFactor) (Variables|Default|GetBoostFactorMorningEveningFactor)) (Utilities|Enum|Equal(Enum) _returnvalue "Night")))
                          (Timeline|Boost PlayFromStart))))
                    (:then_1
                      (if (Variables|Default|GetStyleVar)
                        (CallFunction|FXonKillStatChange)))))))))))
    (:then_2
      ;[1 0.1636 0.04388] Frontend reactivity preview setup
      (if (Reactive|Frontend|IsCosmeticObjectDisplayedinFrontendContext 0)
        (bind _returnvalue_4 (Utilities|Time|SetTimerbyFunctionName 0 "FrontEnd_ReactivityPreview" 7.2 true))
        (Variables|Default|SetFrontEndReactivityTimer _returnvalue_4)
        (else
          (Utilities|Time|ClearandInvalidateTimerbyHandle (Variables|Default|GetFrontEndReactivityTimer)))))))

(event EventOnCurieStateAttached
  (params (Tag :type Gameplay%20Tag%20Structure%20%28by%20ref%29)))

(event EventOnDayPhaseChanged
  (params (CurrentDayPhase :type EFortDayPhase%20Enum :default "Morning") (PreviousDayPhase :type EFortDayPhase%20Enum :default "Morning") (bAtCreation :type Boolean)))

(event EventOnTargetingOrFirstPersonCameraChanged)

(event EventOnCurieStateDetached
  (params (Tag :type Gameplay%20Tag%20Structure%20%28by%20ref%29)))

(event EventOnNewPlayerWeaponType
  (params (NewType :type EFortItemType%20Enum :default "WorldItem") (OldType :type EFortItemType%20Enum :default "WorldItem"))
  (|CallParentFunction NewType OldType) ; This node is disabled and will not be called.\nDrag off pins to build functionality.
)

(event EventOnHitPawn
  (params (SelfActor :type Actor%20Object%20Reference) (OtherActor :type Actor%20Object%20Reference) (NormalImpulse :type Vector :default "0, 0, 0") (Hit :type Hit%20Result%20Structure%20%28by%20ref%29)))

(event EventOnBeginSkydiving
  (|CallParentFunction) ; This node is disabled and will not be called.\nDrag off pins to build functionality.
)

(event EventOnEndSkydiving
  (|CallParentFunction) ; This node is disabled and will not be called.\nDrag off pins to build functionality.
)

(event EventOnWeaponFired)

(event EventOnEnterVehicle
  (params (VehicleEntered :type Fort%20Vehicle%20Interface%20Interface%20%28by%20ref%29))
  (|CallParentFunction VehicleEntered) ; This node is disabled and will not be called.\nDrag off pins to build functionality.
)

(event EventOnExitVehicle
  (|CallParentFunction) ; This node is disabled and will not be called.\nDrag off pins to build functionality.
)

(event Custom|OnKillSFX
  (Audio|SpawnSoundAttached (Variables|Default|GetSound) (CustomCharacter|GetSkeletalMeshForPartOnPlayerPawn)))

(event Custom|FX_OnKillStatChange
  (Utilities|FlowControl|Sequence
    (:then_0
      (bind _asniagara_particle_system_component (Utilities|Casting|CastToNiagaraComponent (CallFunction|GetIdleVFXComponent)))
      (Niagara|SetNiagaraVariable(Bool) _asniagara_particle_system_component "bKillReaction" true)
      (Utilities|FlowControl|RetriggerableDelay 3.0)
      (Niagara|SetNiagaraVariable(Bool) _asniagara_particle_system_component "bKillReaction" false))
    (:then_1
      (PartModifier|SpawnSystem|SpawnSystemAttachedPartModifier "/BRCosmetics/Characters/Player/Male/Medium/Bodies/M_MED_HeavyRoar/FX/Style/NS_MED_HeavyRoar_Style_ReactFX.NS_MED_HeavyRoar_Style_ReactFX" (CustomCharacter|GetSkeletalMeshForPartOnPlayerPawn) "None" "0, 0, 0" "0, 0, 0" "SnapToTarget"))))

(fn ConstructionScript
  (|CallParentFunction))
