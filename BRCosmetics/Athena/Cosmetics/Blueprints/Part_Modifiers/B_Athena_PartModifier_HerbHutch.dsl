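; Part modifier for the HerbHutch character cosmetic that overrides anime outlines for its Superlevel styles.
; After the parent post-resolution event runs, it caches the Athena pawn and reads the loadout's Slot.0 emissive and Progressive stage variant indices.
; These switches govern clearing the body overlay or applying inverted-hull outline overlay materials to Body, Head and Face meshes.
; A note states that the panel style keeps its outlines.
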(event EventOnPostResIn
  (bind _returnvalue (CustomCharacter|GetAssociatedPlayerPawn))
  (bind _returnvalue_2 (Customization|GetSkeletalMeshforPartType _returnvalue "Body" :!class /Script/FortniteGame.FortPlayerPawn))
  ; Superlevel outlines override
  (|CallParentFunction)
  (bind _asplayer_pawn_athena (Utilities|Casting|CastToPlayerPawn_Athena (CustomCharacter|GetAssociatedPlayerPawn)))
  (bind _output_get (Variables|Default|SetCosmeticPawn _asplayer_pawn_athena))
  (bind _index (CosmeticLoadout|GetVariantChannelActiveIndexforSlotActor "(TagName=\"LoadoutSlot.Character\")" _output_get "(TagName=\"Cosmetics.Variant.Channel.Slot.0\")" (Utilities|Array|MakeArray "(TagName=\"Cosmetics.Variant.Property.Emissive0\")" "(TagName=\"Cosmetics.Variant.Property.Emissive1\")" "(TagName=\"Cosmetics.Variant.Property.Emissive2\")" "(TagName=\"Cosmetics.Variant.Property.Emissive3\")") false false))
  (switch int _index
    (:0
      ;; Do not turn off outlines on HerbHutch panel style
      (bind _index_2 (CosmeticLoadout|GetVariantChannelActiveIndexforSlotActor "(TagName=\"LoadoutSlot.Character\")" (Variables|Default|GetCosmeticPawn) "(TagName=\"Cosmetics.Variant.Channel.Progressive\")" (Utilities|Array|MakeArray "(TagName=\"Cosmetics.Variant.Property.Stage1\")" "(TagName=\"Cosmetics.Variant.Property.Stage2\")" "(TagName=\"Cosmetics.Variant.Property.Stage3\")") false false))
      (switch int _index_2
        (:Default
          _)
        (:0
          _)
        (:1
          _))
      (Rendering|Material|SetOverlayMaterial _returnvalue_2 0 false 0 :!class /Script/Engine.MeshComponent))
    (:1
      _)
    (:2
      _)
    (:3
      _))
  (Rendering|Material|SetOverlayMaterial _returnvalue_2 "/BRCosmetics/Characters/Player/Male/Medium/Bodies/M_MED_HerbHutch_Breath/Skins/Superlevels/Materials/MI_SL_S31_AnimeOutline_InvHull.MI_SL_S31_AnimeOutline_InvHull" false 0 :!class /Script/Engine.MeshComponent)
  (Rendering|Material|SetOverlayMaterial (Customization|GetSkeletalMeshforPartType _returnvalue "Head" :!class /Script/FortniteGame.FortPlayerPawn) "/BRCosmetics/Characters/Player/Male/Medium/Bodies/M_MED_HerbHutch_Breath/Skins/Superlevels/Materials/MI_SL_S31_AnimeOutline_InvHull_Head.MI_SL_S31_AnimeOutline_InvHull_Head" false 0 :!class /Script/Engine.MeshComponent)
  (Rendering|Material|SetOverlayMaterial (Customization|GetSkeletalMeshforPartType _returnvalue "Face" :!class /Script/FortniteGame.FortPlayerPawn) "/BRCosmetics/Characters/Player/Male/Medium/Bodies/M_MED_HerbHutch_Breath/Skins/Superlevels/Materials/MI_SL_S31_AnimeOutline_InvHull_FaceAcc.MI_SL_S31_AnimeOutline_InvHull_FaceAcc" false 0 :!class /Script/Engine.MeshComponent))

(fn ConstructionScript
  (|CallParentFunction))
