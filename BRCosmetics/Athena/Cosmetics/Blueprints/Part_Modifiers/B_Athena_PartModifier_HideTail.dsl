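; Athena character part modifier that hides a tail, apparently an idle VFX component, when the player wears a cape or long back bling.
; When customization finishes, it checks back bling through the part's anim instance, deactivating the idle VFX for capes or long items and activating it otherwise.
; After res-in and a short retriggerable delay, it passes the active Progressive variant stage index to the Niagara component's VariantNumber parameter.
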(event EventOnPostResIn
  (|CallParentFunction)
  (Utilities|FlowControl|RetriggerableDelay 0.2)
  (bind _index (CosmeticLoadout|GetVariantChannelActiveIndexforSlotActor "(TagName=\"LoadoutSlot.Character\")" (CustomCharacter|GetAssociatedPlayerPawn) "(TagName=\"Cosmetics.Variant.Channel.Progressive\")" (Utilities|Array|MakeArray "(TagName=\"Cosmetics.Variant.Property.Stage1\")" "(TagName=\"Cosmetics.Variant.Property.Stage2\")" "(TagName=\"Cosmetics.Variant.Property.Stage3\")")))
  (bind _asniagara_particle_system_component (Utilities|Casting|CastToNiagaraComponent (CallFunction|GetIdleVFXComponent)))
  (Effects|Components|ParticleSystem|SetIntParameter _asniagara_particle_system_component "VariantNumber" _index))

(event EventOnCharacterCustomizationFinished
  (params (CharacterPartInfo :type Array%20of%20Applied%20Character%20Part%20Info%20Structures))
  (bind _returnvalue (CallFunction|GetIdleVFXComponent))
  (|CallParentFunction CharacterPartInfo)
  (bind _ascustom_character_part_anim_instance (Utilities|Casting|CastToCustomCharacterPartAnimInstance (Components|SkeletalMesh|GetAnimInstance (Customization|GetSkeletalMeshforPartType (CustomCharacter|GetAssociatedPlayerPawn) (CustomCharacter|GetCharacterPartType) :!class /Script/FortniteGame.FortPlayerPawn))))
  (if (or (.bIsBackblingACape (CustomCharacter|CheckforBackbling _ascustom_character_part_anim_instance 5e+01)) (.bIsBackblingLong (CustomCharacter|CheckforBackbling _ascustom_character_part_anim_instance 5e+01)))
    (Components|Activation|Deactivate _returnvalue)
    (else
      (Components|Activation|Activate _returnvalue true))))

(fn ConstructionScript
  (|CallParentFunction))
