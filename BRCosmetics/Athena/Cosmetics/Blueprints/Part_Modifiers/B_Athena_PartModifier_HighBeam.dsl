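; Cosmetic character part modifier that syncs the part's hidden-hand state to the High Beam pickaxe.
; On weapon change, it casts the new weapon to the High Beam impact pickaxe, copies bIsHandHidden into its "Hand is Hidden" variable and calls ChangeHIddenHandSettings.
; When the part is applied post res-in, it does the same for the equipped weapon from the inventory context, after validity checks.
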(event EventOnWeaponChanged
  (params (NewWeapon :type Fort%20Weapon%20Object%20Reference) (OldWeapon :type Fort%20Weapon%20Object%20Reference))
  (|CallParentFunction NewWeapon OldWeapon)
  (bind _asb_athena_pickaxe_high_beam_impact (Utilities|Casting|CastToB_Athena_Pickaxe_HighBeam_Impact NewWeapon))
  (Variables|/BRCosmetics/Weapons/FORT_Melee/Blueprints/Impact/B_Athena_Pickaxe_HighBeam_Impact.B_Athena_Pickaxe_HighBeam_Impact_C|SetHand%20is%20Hidden (Variables|HiddenPartCheck|GetbIsHandHidden) _asb_athena_pickaxe_high_beam_impact)
  (Class|BAthenaPickaxeHighBeamImpact|ChangeHIddenHandSettings _asb_athena_pickaxe_high_beam_impact))

(event EventOnCosmeticPartAppliedPostResIn
  (params (Part :type Custom%20Character%20Part%20Object%20Reference))
  (bind _returnvalue (BlueprintContext|GetFortInventoryContext))
  (bind _returnvalue_2 (Inventory|GetEquippedWeapon _returnvalue))
  (|CallParentFunction Part)
  (Utilities|IsValid _returnvalue
    (:Is%20Valid
      (Utilities|IsValid _returnvalue_2
        (:Is%20Valid
          (bind _asb_athena_pickaxe_high_beam_impact (Utilities|Casting|CastToB_Athena_Pickaxe_HighBeam_Impact _returnvalue_2))
          (Variables|/BRCosmetics/Weapons/FORT_Melee/Blueprints/Impact/B_Athena_Pickaxe_HighBeam_Impact.B_Athena_Pickaxe_HighBeam_Impact_C|SetHand%20is%20Hidden (Variables|HiddenPartCheck|GetbIsHandHidden) _asb_athena_pickaxe_high_beam_impact)
          (Class|BAthenaPickaxeHighBeamImpact|ChangeHIddenHandSettings _asb_athena_pickaxe_high_beam_impact))))))

(fn ConstructionScript
  (|CallParentFunction))
