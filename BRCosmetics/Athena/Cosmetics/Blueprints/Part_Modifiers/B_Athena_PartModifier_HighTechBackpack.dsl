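; Cosmetic part modifier Blueprint for the HighTechBackpack character part, exposing control over a charge bar in its materials.
; UpdateChargeBarScalar takes a float and writes it to the "ChargeBarScalar" parameter on the materials of the part's skeletal mesh on the player pawn.
; The construction script is empty, and nothing in this Blueprint calls the update or computes the charge value.
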(fn UpdateChargeBarScalar
  (params (Scalar :type Float%20%28double-precision%29))
  (Rendering|Material|SetScalarParameterValueonMaterials (CustomCharacter|GetSkeletalMeshForPartOnPlayerPawn) "ChargeBarScalar" Scalar))

(fn ConstructionScript)
