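; Character part modifier for the HighTower HoneyDew body that drives its transformation materials around the Honeydew emote.
; On customization finish it caches an AnimInstance and reads loadout variant tags into red/green and Stage 2 flags.
; On emote begin, quantized Honeydew montage position sets body TransitionAmt and face element visibility, while the color flag sets hue, saturation and emissive edge.
; Other emote starts and Honeydew emote end set TransitionAmt from the stage flag.
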(event EventOnCharacterCustomizationFinished
  (params (CharacterPartInfo :type Array%20of%20Applied%20Character%20Part%20Info%20Structures))
  (|CallParentFunction CharacterPartInfo)
  (Utilities|FlowControl|Sequence
    (:then_0
      ;[1 1 1] Get Main Skeleton AnimInstance
      (bind _asskeletal_mesh_component (Utilities|Casting|CastToSkeletalMeshComponent (CustomCharacter|GetSkeletalMeshForPartOnPlayerPawn)))
      (bind _asf_med_high_tower_honeydew_anim_bp (Utilities|Casting|CastToF_MED_HighTower_Honeydew_AnimBP (Components|SkeletalMesh|GetAnimInstance _asskeletal_mesh_component)))
      (Variables|Default|SetAnimInstance (Components|SkeletalMesh|GetAnimInstance (CustomCharacter|GetOwnerSkeletalMeshComponent _asf_med_high_tower_honeydew_anim_bp))))
    (:then_1
      ;[1 1 1] Handle Red/Green Variants
      (for _array_element (CosmeticLoadout|GetAllLoadoutVariantInfofromLoadout (CosmeticLoadout|GetCosmeticLoadout))
        (if (Utilities|Equal(Object) (.ItemVariantIsUsedFor (Utilities|Struct|BreakMcpVariantChannelInfo _array_element)) "/BRCosmetics/Athena/Items/Cosmetics/Characters/CID_842_Athena_Commando_F_HightowerHoneydew.CID_842_Athena_Commando_F_HightowerHoneydew")
          (if (GameplayTags|MatchesTag (.ActiveVariantTag (Utilities|Struct|BreakMcpVariantChannelInfo _array_element)) "(TagName=\"Cosmetics.Variant.Property.Mat2\")" true)
            (Variables|Default|SetIsHoneyDewRedVariant true)
            (elif (GameplayTags|MatchesTag (.ActiveVariantTag (Utilities|Struct|BreakMcpVariantChannelInfo _array_element)) "(TagName=\"Cosmetics.Variant.Property.Mat1\")" true)
              (Variables|Default|SetIsHoneyDewRedVariant false)
              (else
                _)))
          (if (GameplayTags|MatchesTag (.ActiveVariantTag (Utilities|Struct|BreakMcpVariantChannelInfo _array_element)) "(TagName=\"Cosmetics.Variant.Property.Stage1\")" true)
            (Variables|Default|SetIsHoneyDewStage2 false)
            (elif (GameplayTags|MatchesTag (.ActiveVariantTag (Utilities|Struct|BreakMcpVariantChannelInfo _array_element)) "(TagName=\"Cosmetics.Variant.Property.Stage2\")" true)
              (Variables|Default|SetIsHoneyDewStage2 true))))))))

(event EventOnBeginEmote
  (params (MontageItemDef :type Fort%20Item%20Definition%20Object%20Reference) (PawnEmoting :type Fort%20Pawn%20Object%20Reference))
  (bind _emoteoutroduration (Variables|Default|GetEmoteOutroDuration))
  (bind _emoteintroduration (Variables|Default|GetEmoteIntroDuration))
  (bind _returnvalue (+ _emoteintroduration _emoteoutroduration))
  (bind _returnvalue_2 (- _returnvalue 0.7))
  (bind _returnvalue_3 (<= (/ (DisplayHelper|RoundFloat (* (Animation|Montage|MontageGetPosition (Variables|Default|GetAnimInstance) (Animation|GetCurrentMontage PawnEmoting :!class /Script/Engine.Character)) (Variables|Default|GetEmoteTimeQuant))) (Variables|Default|GetEmoteTimeQuant)) _returnvalue_2))
  (bind _returnvalue_4 (Math|Float|SelectFloat 1.0 0.0 _returnvalue_3))
  (bind _emotetimequant (Variables|Default|GetEmoteTimeQuant))
  (bind _animinstance (Variables|Default|GetAnimInstance))
  (bind _returnvalue_5 (Animation|Montage|MontageGetPosition _animinstance (Animation|GetCurrentMontage PawnEmoting :!class /Script/Engine.Character)))
  (bind _returnvalue_6 (* _returnvalue_5 _emotetimequant))
  (bind _returnvalue_7 (DisplayHelper|RoundFloat _returnvalue_6))
  (bind _returnvalue_8 (/ _returnvalue_7 _emotetimequant))
  (bind _returnvalue_9 (CustomCharacter|GetSkeletalMeshForPartOnPlayerPawn))
  (bind _returnvalue_10 (Animation|GetCurrentMontage PawnEmoting :!class /Script/Engine.Character))
  (bind _mid (Material|GetMID _returnvalue_9))
  (bind _returnvalue_11 (CustomCharacter|GetSkeletalMeshForPartOnPlayerPawn))
  (bind _mid_2 (Material|GetMID _returnvalue_11))
  (bind (_head_skeletal_mesh_mid _face_skeletal_mesh_mat_ind_0 _face_skeletal_mesh_mat_ind_1) (CallFunction|GetMIDsFromParts))
  (if (Utilities|Equal(Object) _returnvalue_10 "/BRCosmetics/Animation/Game/MainPlayer/Emotes/Honeydew/Honeydew_M.Honeydew_M")
    (if (<= _returnvalue_8 (Variables|Default|GetEmoteIntroDuration))
      (bind _asmaterial_instance_dynamic (Utilities|Casting|CastToMaterialInstanceDynamic _face_skeletal_mesh_mat_ind_0))
      (bind _asmaterial_instance_dynamic_2 (Utilities|Casting|CastToMaterialInstanceDynamic _face_skeletal_mesh_mat_ind_1))
      (Rendering|Material|SetScalarParameterValue _asmaterial_instance_dynamic "Use Custom Opacity" _returnvalue_4 :!class /Script/Engine.MaterialInstanceDynamic)
      (Rendering|Material|SetScalarParameterValue _asmaterial_instance_dynamic_2 "Hide Element 09" _returnvalue_4 :!class /Script/Engine.MaterialInstanceDynamic)
      (Rendering|Material|SetScalarParameterValue _mid_2 "TransitionAmt" (Math|Float|MapRangeClamped _returnvalue_8 0.75 2.75 0.0 1.0) :!class /Script/Engine.MaterialInstanceDynamic)
      (else
        (Rendering|Material|SetScalarParameterValue _mid_2 "TransitionAmt" (Math|Float|MapRangeClamped _returnvalue_8 4.0 6.0 1.0) :!class /Script/Engine.MaterialInstanceDynamic)))
    (if (Variables|Default|GetIsHoneyDewRedVariant)
      (Rendering|Material|SetScalarParameterValue _mid "Hue" -0.03 :!class /Script/Engine.MaterialInstanceDynamic)
      (Rendering|Material|SetScalarParameterValue _mid "Saturation" 2.35 :!class /Script/Engine.MaterialInstanceDynamic)
      (Rendering|Material|SetVectorParameterValue _mid "EmissiveEdge" "(R=0.498039,G=0.000000,B=0.000000,A=1.000000)" :!class /Script/Engine.MaterialInstanceDynamic)
      (else
        (Rendering|Material|SetScalarParameterValue _mid "Hue" 0.2 :!class /Script/Engine.MaterialInstanceDynamic)
        (Rendering|Material|SetScalarParameterValue _mid "Saturation" 0.6 :!class /Script/Engine.MaterialInstanceDynamic)
        (Rendering|Material|SetVectorParameterValue _mid "EmissiveEdge" "(R=0.243137,G=0.498039,B=0.160784,A=1.000000)" :!class /Script/Engine.MaterialInstanceDynamic)))
    (else
      (Rendering|Material|SetScalarParameterValue (Material|GetMID (CustomCharacter|GetSkeletalMeshForPartOnPlayerPawn)) "TransitionAmt" (Math|Float|SelectFloat 1.0 0.0 (Variables|Default|GetIsHoneyDewStage2)) :!class /Script/Engine.MaterialInstanceDynamic))))

(event EventOnEndEmote
  (params (MontageItemDef :type Fort%20Item%20Definition%20Object%20Reference) (PawnEmoting :type Fort%20Pawn%20Object%20Reference))
  (if (Utilities|Equal(Object) (Animation|GetCurrentMontage PawnEmoting :!class /Script/Engine.Character) "/BRCosmetics/Animation/Game/MainPlayer/Emotes/Honeydew/Honeydew_M.Honeydew_M")
    (Rendering|Material|SetScalarParameterValue (Material|GetMID (CustomCharacter|GetSkeletalMeshForPartOnPlayerPawn)) "TransitionAmt" (Math|Float|SelectFloat 1.0 0.0 (Variables|Default|GetIsHoneyDewStage2)) :!class /Script/Engine.MaterialInstanceDynamic)))

(fn GetMIDsfromParts
  (return (Material|GetMID (Customization|GetSkeletalMeshforPartType (CustomCharacter|GetAssociatedPlayerPawn) "Head" :!class /Script/FortniteGame.FortPlayerPawn)) (Rendering|Material|GetMaterial (Customization|GetSkeletalMeshforPartType (CustomCharacter|GetAssociatedPlayerPawn) "Face" :!class /Script/FortniteGame.FortPlayerPawn) 0) (Rendering|Material|GetMaterial (Customization|GetSkeletalMeshforPartType (CustomCharacter|GetAssociatedPlayerPawn) "Face" :!class /Script/FortniteGame.FortPlayerPawn) 1) (Material|GetMID (Customization|GetSkeletalMeshforPartType (CustomCharacter|GetAssociatedPlayerPawn) "Body" :!class /Script/FortniteGame.FortPlayerPawn))))

(fn ConstructionScript
  (|CallParentFunction))
