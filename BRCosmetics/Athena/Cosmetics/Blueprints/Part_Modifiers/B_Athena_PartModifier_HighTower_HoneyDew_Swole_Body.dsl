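; Character part modifier for the HighTower HoneyDew Swole body that sets its transformation material look around the Honeydew emote.
; On emote begin, it reads the montage position through the owner player's anim instance and quantizes it.
; That time drives TransitionAmt and a sine-shaped vein pulse on the part's MID, with separate intro and post-intro mappings.
; Other emotes set TransitionAmt to full, as does emote end while Honeydew is the current montage.
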(event EventOnBeginEmote
  (params (MontageItemDef :type Fort%20Item%20Definition%20Object%20Reference) (PawnEmoting :type Fort%20Pawn%20Object%20Reference))
  (bind _emotetimequant (Variables|Default|GetEmoteTimeQuant))
  (bind _returnvalue (Animation|GetCurrentMontage PawnEmoting :!class /Script/Engine.Character))
  (bind _animinstance (CallFunction|GetAnimInstance))
  (bind _returnvalue_2 (Animation|Montage|MontageGetPosition _animinstance _returnvalue))
  (bind _returnvalue_3 (* _returnvalue_2 _emotetimequant))
  (bind _returnvalue_4 (DisplayHelper|RoundFloat _returnvalue_3))
  (bind _returnvalue_5 (/ _returnvalue_4 _emotetimequant))
  (bind _returnvalue_6 (Math|Float|MapRangeClamped _returnvalue_5 4.0 6.0 1.0))
  (bind _returnvalue_7 (CustomCharacter|GetSkeletalMeshForPartOnPlayerPawn))
  (bind _mid (Material|GetMID _returnvalue_7))
  (if (Utilities|Equal(Object) _returnvalue "/BRCosmetics/Animation/Game/MainPlayer/Emotes/Honeydew/Honeydew_M.Honeydew_M")
    (if (<= _returnvalue_5 (Variables|Default|GetEmoteIntroDuration))
      (Rendering|Material|SetScalarParameterValue _mid "TransitionAmt" (Math|Float|MapRangeClamped _returnvalue_5 0.75 2.75 0.0 1.0) :!class /Script/Engine.MaterialInstanceDynamic)
      (Rendering|Material|SetScalarParameterValue _mid "BP_VeinsIntensity" (* (+ (Math|Trig|Sin(Radians) (* (- (Math|Float|MapRangeClamped _returnvalue_5 0.25 3.91 0.0 1.0) 0.25) 6.2831)) 1.0) 0.5) :!class /Script/Engine.MaterialInstanceDynamic)
      (else
        (Rendering|Material|SetScalarParameterValue _mid "TransitionAmt" _returnvalue_6 :!class /Script/Engine.MaterialInstanceDynamic)
        (Rendering|Material|SetScalarParameterValue _mid "BP_VeinsIntensity" (* (+ (Math|Trig|Sin(Radians) (* (- _returnvalue_6 0.25) 6.2831)) 1.0) 0.35) :!class /Script/Engine.MaterialInstanceDynamic)))
    (else
      (Rendering|Material|SetScalarParameterValue (Material|GetMID (CustomCharacter|GetSkeletalMeshForPartOnPlayerPawn)) "TransitionAmt" 1.0 :!class /Script/Engine.MaterialInstanceDynamic))))

(event EventOnEndEmote
  (params (MontageItemDef :type Fort%20Item%20Definition%20Object%20Reference) (PawnEmoting :type Fort%20Pawn%20Object%20Reference))
  (if (Utilities|Equal(Object) (Animation|GetCurrentMontage PawnEmoting :!class /Script/Engine.Character) "/BRCosmetics/Animation/Game/MainPlayer/Emotes/Honeydew/Honeydew_M.Honeydew_M")
    (Rendering|Material|SetScalarParameterValue (Material|GetMID (CustomCharacter|GetSkeletalMeshForPartOnPlayerPawn)) "TransitionAmt" 1.0 :!class /Script/Engine.MaterialInstanceDynamic)))

(fn GetAnimInstance
  (return (CustomCharacter|GetOwnerFortPlayerAnimInstance (Utilities|Casting|CastToF_MED_HighTower_Honeydew_Swole_AnimBP (Components|SkeletalMesh|GetAnimInstance (Customization|GetSkeletalMeshforPartType (CustomCharacter|GetAssociatedPlayerPawn) (CustomCharacter|GetCharacterPartType) :!class /Script/FortniteGame.FortPlayerPawn))))))

(fn ConstructionScript
  (|CallParentFunction))
