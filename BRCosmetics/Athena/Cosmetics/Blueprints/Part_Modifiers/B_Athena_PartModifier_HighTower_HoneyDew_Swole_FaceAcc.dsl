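; Character part modifier for the HighTower HoneyDew Swole face accessory that drives its material's TransitionAmt parameter around the Honeydew emote.
; On emote begin, it reads the montage position from the owner's anim instance via the accessory's AnimBP and snaps it to quantized steps.
; That time is range-mapped separately for intro and remainder to set TransitionAmt.
; Other emotes set it to 1, as does emote end while Honeydew is current.
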(event EventOnBeginEmote
  (params (MontageItemDef :type Fort%20Item%20Definition%20Object%20Reference) (PawnEmoting :type Fort%20Pawn%20Object%20Reference))
  (bind _returnvalue (CustomCharacter|GetSkeletalMeshForPartOnPlayerPawn))
  (bind _mid (Material|GetMID _returnvalue))
  (bind _animinstance (CallFunction|GetAnimInstance))
  (bind _returnvalue_2 (Animation|GetCurrentMontage PawnEmoting :!class /Script/Engine.Character))
  (bind _returnvalue_3 (Animation|Montage|MontageGetPosition _animinstance _returnvalue_2))
  (bind _emotetimequant (Variables|Default|GetEmoteTimeQuant))
  (bind _returnvalue_4 (* _returnvalue_3 _emotetimequant))
  (bind _returnvalue_5 (DisplayHelper|RoundFloat _returnvalue_4))
  (bind _returnvalue_6 (/ _returnvalue_5 _emotetimequant))
  (bind _emoteintroduration (Variables|Default|GetEmoteIntroDuration))
  (if (Utilities|Equal(Object) _returnvalue_2 "/BRCosmetics/Animation/Game/MainPlayer/Emotes/Honeydew/Honeydew_M.Honeydew_M")
    (if (<= _returnvalue_6 _emoteintroduration)
      (Rendering|Material|SetScalarParameterValue _mid "TransitionAmt" (Math|Float|MapRangeClamped _returnvalue_6 (Variables|Default|GetEmoteTimeStart) _emoteintroduration 0.0 1.0) :!class /Script/Engine.MaterialInstanceDynamic)
      (else
        (Rendering|Material|SetScalarParameterValue _mid "TransitionAmt" (Math|Float|MapRangeClamped _returnvalue_6 _emoteintroduration (Variables|Default|GetEmoteTimeEnd) 1.0) :!class /Script/Engine.MaterialInstanceDynamic)))
    (else
      (Rendering|Material|SetScalarParameterValue (Material|GetMID (CustomCharacter|GetSkeletalMeshForPartOnPlayerPawn)) "TransitionAmt" 1.0 :!class /Script/Engine.MaterialInstanceDynamic))))

(event EventOnEndEmote
  (params (MontageItemDef :type Fort%20Item%20Definition%20Object%20Reference) (PawnEmoting :type Fort%20Pawn%20Object%20Reference))
  (if (Utilities|Equal(Object) (Animation|GetCurrentMontage PawnEmoting :!class /Script/Engine.Character) "/BRCosmetics/Animation/Game/MainPlayer/Emotes/Honeydew/Honeydew_M.Honeydew_M")
    (Rendering|Material|SetScalarParameterValue (Material|GetMID (CustomCharacter|GetSkeletalMeshForPartOnPlayerPawn)) "TransitionAmt" 1.0 :!class /Script/Engine.MaterialInstanceDynamic)))

(fn GetAnimInstance
  (return (CustomCharacter|GetOwnerFortPlayerAnimInstance (Utilities|Casting|CastToF_MED_HighTower_Honeydew_Swole_FaceAcc_AnimBP (Components|SkeletalMesh|GetAnimInstance (Customization|GetSkeletalMeshforPartType (CustomCharacter|GetAssociatedPlayerPawn) (CustomCharacter|GetCharacterPartType) :!class /Script/FortniteGame.FortPlayerPawn))))))

(fn ConstructionScript
  (|CallParentFunction))
