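; Athena cosmetic part modifier Blueprint for HighTower Squash that reacts to tracked stat changes through the idle VFX.
; On a stat change it calls the parent handler, then converts the stat value to a float and writes it to the idle VFX component's User.StatChangedKills parameter.
; The state name is not checked, so every stat change drives that kills parameter, and the construction script only calls its parent.
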(event EventOnStatChanged
  (params (State%20Name :type Name :default "None") (Stat%20Value :type Integer))
  (|CallParentFunction State%20Name Stat%20Value)
  (Effects|Components|ParticleSystem|SetFloatParameter (CallFunction|GetIdleVFXComponent) "User.StatChangedKills" (Math|Conversions|ToFloat(Integer) Stat%20Value)))

(fn ConstructionScript
  (|CallParentFunction))
