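; Cosmetic part modifier for the HighTower Tomato back bling that shows a kill counter and toggles a targeting effect with sound.
; Watches the AthenaKills stat and writes it to the mesh's Counter_InputNumber material parameter.
; Targeting changes switch the idle VFX and play ADS start or end audio.
; Tick, enabled only in frontend, repeats that switch every frame from the pack AnimBP's stage_2 value; female skeletons also get the frontend-preview VFX flag.
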(event EventOnPostResIn
  (bind _returnvalue (CustomCosmeticModifier|IsinFrontEnd))
  (bind (_ismale _isfemale) (FortPlayerPawn|GetGender (CustomCharacter|GetAssociatedPlayerPawn)))
  (|CallParentFunction)
  (Utilities|FlowControl|Sequence
    (:then_0
      ;[1 1 1] Kill Counter
      (Variables|Default|SetkillCount 0.0)
      (bind (_statname _statvalue) (Cosmetics|WatchCosmeticStat (CustomCharacter|GetAssociatedPlayerPawn) "(TagName=\"StatManager.AthenaKills\")")
        (:OnStatChanged
          (Variables|Default|SetkillCount (Math|Conversions|ToFloat(Integer) _statvalue))
          (Rendering|Material|SetScalarParameterValueonMaterials (CustomCharacter|GetSkeletalMeshForPartOnPlayerPawn) "Counter_InputNumber" (Variables|Default|GetkillCount)))))
    (:then_1
      ;[1 1 1] Set Tick Enabled for Frontend only Animations and solved Edge Case with Targeting Laser on Female Skeletons
      (Actor|Tick|SetActorTickEnabled _returnvalue)
      (Effects|Components|ParticleSystem|SetBoolParameter (CallFunction|GetIdleVFXComponent) "bIsFrontEndPreview" (or _returnvalue _isfemale)))))

(event EventOnTargetingOrFirstPersonCameraChanged
  (bind _returnvalue (CallFunction|GetIdleVFXComponent))
  (bind _audio (Variables|Default|GetAudio))
  (bind _audio_2 (Variables|Default|GetAudio))
  ;[1 1 1] Targeting Check and setup
  (|CallParentFunction)
  (if (Variables|/Script/FortniteGame.FortPlayerPawn|GetbIsTargeting (CustomCharacter|GetAssociatedPlayerPawn))
    (Components|Activation|Activate _returnvalue false)
    (Effects|Components|ParticleSystem|SetBoolParameter _returnvalue "Activate" true)
    (Audio|Components|Audio|SetSound _audio (Variables|Default|GetSoundOnADSStart))
    (Audio|Components|Audio|Play _audio 0.0)
    (else
      (Effects|Components|ParticleSystem|SetBoolParameter _returnvalue "Activate" false)
      (Components|Activation|Deactivate _returnvalue)
      (Audio|Components|Audio|SetSound _audio_2 (Variables|Default|GetSoundOnADSEnd))
      (Audio|Components|Audio|Play _audio_2 0.0))))

(event EventTick
  (params (DeltaSeconds :type Float%20%28single-precision%29))
  (bind _returnvalue (CallFunction|GetIdleVFXComponent))
  (bind _audio (Variables|Default|GetAudio))
  (bind _audio_2 (Variables|Default|GetAudio))
  ;[1 1 1] Frontend Animation
  (bind _asm_med_high_tower_tomato_pack_anim_bp (Utilities|Casting|CastToM_MED_HighTower_Tomato_Pack_AnimBP (Components|SkeletalMesh|GetAnimInstance (Customization|GetSkeletalMeshforPartType (CustomCharacter|GetAssociatedPlayerPawn) (CustomCharacter|GetCharacterPartType) :!class /Script/FortniteGame.FortPlayerPawn))))
  (if (>= (Variables|/BRCosmetics/Accessories/FORT_Backpacks/Backpack_M_MED_HighTower_Tomato/Meshes/M_MED_HighTower_Tomato_Pack_AnimBP.M_MED_HighTower_Tomato_Pack_AnimBP_C|Getstage_2 _asm_med_high_tower_tomato_pack_anim_bp) 1.0)
    ;[1 1 1] Targeting Check and setup
    (Components|Activation|Activate _returnvalue false)
    (Effects|Components|ParticleSystem|SetBoolParameter _returnvalue "Activate" true)
    (Audio|Components|Audio|SetSound _audio (Variables|Default|GetSoundOnADSStart))
    (Audio|Components|Audio|Play _audio 0.0)
    (else
      (Effects|Components|ParticleSystem|SetBoolParameter _returnvalue "Activate" false)
      (Components|Activation|Deactivate _returnvalue)
      (Audio|Components|Audio|SetSound _audio_2 (Variables|Default|GetSoundOnADSEnd))
      (Audio|Components|Audio|Play _audio_2 0.0))))

(fn ConstructionScript)
