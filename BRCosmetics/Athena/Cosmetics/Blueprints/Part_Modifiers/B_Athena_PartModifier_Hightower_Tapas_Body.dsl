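; Part modifier Blueprint for the Hightower Tapas body cosmetic that attaches an idle eye effect to the character.
; On the post-resource-in event, after the parent logic, it snaps the FX_Eyes_Idle component to the Head of the player pawn's skeletal mesh for the part.
; It then adds that effect to the array of additional FX needing hidden-part checks.
; The construction script only calls the parent.
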(event EventOnPostResIn
  (|CallParentFunction)
  (Transformation|AttachComponentToComponent (Variables|Default|GetFX_Eyes_Idle) (CustomCharacter|GetSkeletalMeshForPartOnPlayerPawn) "Head" "SnapToTarget" "SnapToTarget" "KeepRelative" true)
  (Utilities|Array|Add (Variables|HiddenPartCheck|GetAdditional%20FX%20Needing%20Part%20Checks) (Variables|Default|GetFX_Eyes_Idle)))

(fn ConstructionScript
  (|CallParentFunction))
