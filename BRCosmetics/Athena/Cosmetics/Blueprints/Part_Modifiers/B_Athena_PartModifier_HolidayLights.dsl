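; Cosmetic part modifier for Holiday Lights that tunes character material parameters for the front end or in-match and for Lumen lighting.
; After res-in it caches the pawn's head, body and face meshes, then updates them and rebinds updates to user-settings UI changes.
; Global illumination quality above 1 counts as Lumen, selecting LumenFix base color, metallic, emissive and fresnel values per context.
; Debug PrintString calls run during each update.
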(event EventOnPostResIn
  (|CallParentFunction)
  (Default|GetParts)
  (bind _returnvalue (Settings|GetGameUserSettings))
  (EventDispatchers|BindEventtoOnGameUserSettingsUINeedsUpdate _returnvalue (EventDispatchers|CreateEvent :!event UpdateHolidayLightsParameters))
  (CallFunction|UpdateHolidayLightsParameters))

(fn GetParts
  (bind _returnvalue (CustomCharacter|GetAssociatedPlayerPawn))
  (Variables|Default|Sethead (Customization|GetSkeletalMeshforPartType _returnvalue "Head" :!class /Script/FortniteGame.FortPlayerPawn))
  (Variables|Default|Setbody (Customization|GetSkeletalMeshforPartType _returnvalue "Body" :!class /Script/FortniteGame.FortPlayerPawn))
  (Variables|Default|Setface (Customization|GetSkeletalMeshforPartType _returnvalue "Face" :!class /Script/FortniteGame.FortPlayerPawn)))

(fn isLumen
  (bind _returnvalue (Settings|GetGameUserSettings))
  (Development|PrintString (Utilities|String|Append "Lumen " (Utilities|String|ToString(Integer) (Settings|GetGlobalIlluminationQuality _returnvalue)) " " (Utilities|String|ToString(Boolean) (> (Settings|GetGlobalIlluminationQuality _returnvalue) 1))))
  (return (> (Settings|GetGlobalIlluminationQuality _returnvalue) 1)))

(fn SetScalarParameterOnMaterialsMultiParts
  (params (name :type Name) (value :type Float%20%28double-precision%29))
  (Rendering|Material|SetScalarParameterValueonMaterials (Variables|Default|Gethead) name value)
  (Rendering|Material|SetScalarParameterValueonMaterials (Variables|Default|Getface) name value)
  (Rendering|Material|SetScalarParameterValueonMaterials (Variables|Default|Getbody) name value))

(fn UpdateHolidayLightsParameters
  (Development|PrintString (Utilities|GetDisplayName self))
  (Default|SetScalarParameteronMaterialsMultiParts "IsInFrontEnd" (Math|Float|SelectFloat 1.0 0.0 (CustomCosmeticModifier|IsinFrontEnd)))
  ;[1 1 1] lumen check
  (bind _lumen (CallFunction|IsLumen))
  (if _lumen
    (if (CustomCosmeticModifier|IsinFrontEnd)
      ;[1 1 1] lumen front end
      (Default|SetScalarParameteronMaterialsMultiParts "LumenFix_BaseColorMult" 0.4)
      (Default|SetScalarParameteronMaterialsMultiParts "LumenFix_MetallicMult" 0.3)
      (Default|SetScalarParameteronMaterialsMultiParts "LumenFix_EmissiveMult" 0.5)
      (Default|SetScalarParameteronMaterialsMultiParts "LumenFix_FresnelAdd")
      (else
        ;[1 1 1] lumen BR
        (Default|SetScalarParameteronMaterialsMultiParts "LumenFix_BaseColorMult" 0.75)
        (Default|SetScalarParameteronMaterialsMultiParts "LumenFix_MetallicMult" 0.75)
        (Default|SetScalarParameteronMaterialsMultiParts "LumenFix_EmissiveMult" 0.75)
        (Default|SetScalarParameteronMaterialsMultiParts "LumenFix_FresnelAdd")))
    (else
      (Development|PrintString "No GI")
      (if (CustomCosmeticModifier|IsinFrontEnd)
        ;[1 1 1] not lumen FE
        (Default|SetScalarParameteronMaterialsMultiParts "LumenFix_BaseColorMult" 1.0)
        (Default|SetScalarParameteronMaterialsMultiParts "LumenFix_MetallicMult" 0.5)
        (Default|SetScalarParameteronMaterialsMultiParts "LumenFix_EmissiveMult" 1.0)
        (Default|SetScalarParameteronMaterialsMultiParts "LumenFix_FresnelAdd" 0.15)
        (else
          ;[1 1 1] not lumen BR
          (Default|SetScalarParameteronMaterialsMultiParts "LumenFix_BaseColorMult" 1.0)
          (Default|SetScalarParameteronMaterialsMultiParts "LumenFix_MetallicMult" 1.0)
          (Default|SetScalarParameteronMaterialsMultiParts "LumenFix_EmissiveMult" 1.0)
          (Default|SetScalarParameteronMaterialsMultiParts "LumenFix_FresnelAdd"))))))

(fn ConstructionScript
  (|CallParentFunction))
