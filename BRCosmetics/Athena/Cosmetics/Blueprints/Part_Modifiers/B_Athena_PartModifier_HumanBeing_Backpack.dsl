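; Cosmetic part modifier for the HumanBeing backpack that plays a material-driven "death" animation when the damage-received stat of the player pawn changes.
; On post res-in, it caches the part's skeletal mesh and watches the StatManager.DamageReceived cosmetic stat.
; Each change sets PlayDeathAnim and restarts a timeline that drives the Y_Pos and DeathVisibility material scalars.
; PlayDeathAnim resets to zero when the timeline finishes.
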(event EventOnPostResIn
  (bind _returnvalue (CustomCharacter|GetSkeletalMeshForPartOnPlayerPawn))
  (|CallParentFunction)
  (bind (_statname _statvalue) (Cosmetics|WatchCosmeticStat (CustomCharacter|GetAssociatedPlayerPawn) "(TagName=\"StatManager.DamageReceived\")" 0 false)
    (:OnStatChanged
      (Rendering|Material|SetScalarParameterValueonMaterials _returnvalue "PlayDeathAnim" 1.0)
      (bind (_direction _y _visibility) (Timeline|Death%20Animation PlayFromStart)
        (:Update
          (Rendering|Material|SetScalarParameterValueonMaterials _returnvalue "Y_Pos" _y)
          (Rendering|Material|SetScalarParameterValueonMaterials _returnvalue "DeathVisibility" _visibility))
        (:Finished
          (Rendering|Material|SetScalarParameterValueonMaterials _returnvalue "PlayDeathAnim" 0.0))))))

(fn ConstructionScript
  (|CallParentFunction))
