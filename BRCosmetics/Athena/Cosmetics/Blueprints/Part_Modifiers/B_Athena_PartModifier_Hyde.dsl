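; Character part modifier for the Hyde cosmetic that sets the HatAndCoatTransformCurve material scalar on its part mesh around emotes.
; On emote begin, it quantizes the montage position and, for the ManAndMonster ToMonster montage, maps it through one of two clamped ramps split at the intro duration.
; Other montages set the scalar to one while the anim instance is valid.
; Emote end restores one if that montage is still current.
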(event EventOnBeginEmote
  (params (MontageItemDef :type Fort%20Item%20Definition%20Object%20Reference) (PawnEmoting :type Fort%20Pawn%20Object%20Reference))
  (bind _anim_instance (CallFunction|GetAnimInstance))
  (bind _returnvalue (Animation|GetCurrentMontage PawnEmoting :!class /Script/Engine.Character))
  (bind _returnvalue_2 (Animation|Montage|MontageGetPosition _anim_instance _returnvalue))
  (bind _emotetimequant (Variables|Default|GetEmoteTimeQuant))
  (bind _returnvalue_3 (* _returnvalue_2 _emotetimequant))
  (bind _returnvalue_4 (DisplayHelper|RoundFloat _returnvalue_3))
  (bind _returnvalue_5 (/ _returnvalue_4 _emotetimequant))
  (bind _returnvalue_6 (CustomCharacter|GetSkeletalMeshForPartOnPlayerPawn))
  (Utilities|IsValid _anim_instance
    (:Is%20Valid
      (if (Utilities|Equal(Object) _returnvalue "/BRCosmetics/Animation/Game/MainPlayer/Emotes/ManAndMonster/Emote_ManAndMonster_ToMonster_CMM_M.Emote_ManAndMonster_ToMonster_CMM_M")
        (if (<= _returnvalue_5 (Variables|Default|GetEmoteIntroDuration))
          (Rendering|Material|SetScalarParameterValueonMaterials _returnvalue_6 "HatAndCoatTransformCurve" (Math|Float|MapRangeClamped _returnvalue_5 5.8 6.3 0.0 1.0))
          (else
            (Rendering|Material|SetScalarParameterValueonMaterials _returnvalue_6 "HatAndCoatTransformCurve" (Math|Float|MapRangeClamped _returnvalue_5 7.86 8.36 1.0))))
        (else
          (Rendering|Material|SetScalarParameterValueonMaterials (CustomCharacter|GetSkeletalMeshForPartOnPlayerPawn) "HatAndCoatTransformCurve" 1.0))))))

(event EventOnEndEmote
  (params (MontageItemDef :type Fort%20Item%20Definition%20Object%20Reference) (PawnEmoting :type Fort%20Pawn%20Object%20Reference))
  (if (Utilities|Equal(Object) (Animation|GetCurrentMontage PawnEmoting :!class /Script/Engine.Character) "/BRCosmetics/Animation/Game/MainPlayer/Emotes/ManAndMonster/Emote_ManAndMonster_ToMonster_CMM_M.Emote_ManAndMonster_ToMonster_CMM_M")
    (Rendering|Material|SetScalarParameterValueonMaterials (CustomCharacter|GetSkeletalMeshForPartOnPlayerPawn) "HatAndCoatTransformCurve" 1.0)))

(fn GetAnimInstance
  (return (Components|SkeletalMesh|GetAnimInstance (Customization|GetSkeletalMeshforPartType (CustomCharacter|GetAssociatedPlayerPawn) (CustomCharacter|GetCharacterPartType) :!class /Script/FortniteGame.FortPlayerPawn))))

(fn ConstructionScript
  (|CallParentFunction))
