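; Cosmetic part modifier for the HydraTrumpetCoach backpack that drives glowing-eye and audio reactivity.
; A glow pulse stamps game time on materials and briefly enables idle particles, once after a delay in frontend or preview, otherwise immediately on init.
; In gameplay, stat watchers play audio for opened chests and shield gains.
; Shield audio is gated by a bool, re-enabled by timer or quick-healing ability tags, to prevent spam.
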(event Custom|FrontEndReactivityLoop
  ;[0.4766 0.4867 1] FrontEnd Reactivity Loop
  (Utilities|FlowControl|Delay 7.5)
  (Rendering|Material|SetScalarParameterValueonMaterials (CustomCharacter|GetSkeletalMeshForPartOnPlayerPawn) "BP_GetGameTime" (Utilities|Time|GetGameTimeinSeconds))
  (Effects|Components|ParticleSystem|SetBoolParameter (CallFunction|GetIdleVFXComponent) "Reactive" true)
  (Utilities|FlowControl|Delay 2.2)
  (Effects|Components|ParticleSystem|SetBoolParameter (CallFunction|GetIdleVFXComponent) "Reactive" false))

(event EventOnPostResIn
  (bind _returnvalue (CustomCharacter|GetAssociatedPlayerPawn))
  (|CallParentFunction)
  (if (or (Reactive|Frontend|IsCosmeticObjectDisplayedinFrontendContext 0) (Reactive|Frontend|IsCosmeticObjectDisplayedasPreview 0))
    (CallFunction|FrontEndReactivityLoop)
    (else
      (Utilities|FlowControl|Sequence
        (:then_0
          (Cosmetics|WatchCosmeticStat _returnvalue "(TagName=\"StatManager.OpenChests\")" 0 false
            (:OnStatChanged
              (CallFunction|AudioStatWatcher "None" 1 (Variables|Default|GetAudioReactivityTag)))))
        (:then_1
          (Cosmetics|WatchCosmeticStat _returnvalue "(TagName=\"StatManager.GainedShieldTimes\")" 0 false
            (:OnStatChanged
              (bind _returnvalue_2 (Utilities|Time|SetTimerbyEvent (delegate Custom|EnableAudio) 1.0))
              (Variables|Default|SetAudioTimer _returnvalue_2)
              (CallFunction|PlayShieldReactiveAudio)))))
      (CallFunction|BackpackReactivity))))

(event EventEndPlay
  (params (EndPlayReason :type EEndPlayReason%20Enum :default "Destroyed"))
  (bind _audiotimer (Variables|Default|GetAudioTimer))
  (|CallParentFunction EndPlayReason)
  (if (Utilities|Time|IsValidTimerHandle _audiotimer)
    (Utilities|Time|ClearandInvalidateTimerbyHandle _audiotimer)))

(event Custom|EnableAudio
  ;[0 0 0] Sets a bool to prevent reactive audio spamming when consuming shields
  (Variables|Default|SetbAllowReactiveAudio true))

(event Custom|BackpackReactivity
  ;[1 0.7581 0.1451] Glowing Eyes Reactivity
  (Rendering|Material|SetScalarParameterValueonMaterials (CustomCharacter|GetSkeletalMeshForPartOnPlayerPawn) "BP_GetGameTime" (Utilities|Time|GetGameTimeinSeconds))
  (Effects|Components|ParticleSystem|SetBoolParameter (CallFunction|GetIdleVFXComponent) "Reactive" true)
  (Utilities|FlowControl|Delay 2.2)
  (Effects|Components|ParticleSystem|SetBoolParameter (CallFunction|GetIdleVFXComponent) "Reactive" false))

(fn AllowShieldAudio
  (params (GameplayAbility :type Fort%20Gameplay%20Ability%20Object%20Reference) (AbilityTags :type Gameplay%20Tag%20Container%20Structure))
  (if (GameplayTags|MatchesAnyTags "(TagName=\"Gameplay.Action.Player.QuickHealingItem\")" AbilityTags true)
    (Variables|Default|SetbAllowReactiveAudio true)))

(fn PlayShieldReactiveAudio
  (if (Variables|Default|GetbAllowReactiveAudio)
    (CallFunction|AudioStatWatcher "None" 1 (Variables|Default|GetAudioReactivityTag))
    (Variables|Default|SetbAllowReactiveAudio false)))

(fn ConstructionScript)
