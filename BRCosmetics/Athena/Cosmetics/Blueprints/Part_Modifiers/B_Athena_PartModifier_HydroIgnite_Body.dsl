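; Body part modifier for the HydroIgnite outfit that shows lightning effects when its matching pickaxe is in the loadout.
; On res-in it creates a dynamic body material and compares the loadout pickaxe to PickaxeID.
; Tick caches weapon type and toggles idle VFX and material opacity by type, pickaxe match and holster state.
; Weapon changes wait a tick, then for another type spawn delayed Niagara hand lightning on the body mesh.
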(event EventOnPostResIn
  (bind (_returnvalue_bannericonid _returnvalue_bannercolorid _returnvalue_skydivecontrail _returnvalue_glider _returnvalue_pickaxe) (Utilities|Struct|BreakFortAthenaLoadout (Cosmetics|GetPawnCosmeticLoadout (CustomCharacter|GetAssociatedPlayerPawn) false true)))
  (bind _returnvalue (CustomCharacter|GetSkeletalMeshForPartOnPlayerPawn))
  (|CallParentFunction)
  (bind _returnvalue_2 (Rendering|Material|CreateDynamicMaterialInstance _returnvalue 0 (Rendering|Material|GetMaterial _returnvalue 0) "None" :!class /Script/Engine.PrimitiveComponent))
  (bind _output_get (Variables|Default|SetLightningDMI _returnvalue_2))
  (Rendering|Material|SetMaterial _returnvalue 0 _output_get)
  (Variables|Default|SetPickaxeCheck (Utilities|Equal(Object) _returnvalue_pickaxe (Variables|Default|GetPickaxeID))))

(event EventOnWeaponChanged
  (params (NewWeapon :type Fort%20Weapon%20Object%20Reference) (OldWeapon :type Fort%20Weapon%20Object%20Reference))
  (|CallParentFunction NewWeapon OldWeapon)
  (Utilities|FlowControl|DelayUntilNextTick)
  (if (and (== (Math|Conversions|ToInteger(Byte) (Variables|Default|GetWeaponType)) 24) (Variables|Default|GetPickaxeCheck))
    (Utilities|FlowControl|Delay 0.5)
    (Niagara|SpawnSystemAttached "/BRCosmetics/Characters/Player/Male/Medium/Bodies/M_MED_HydroIgnite/FX/NS_HydroIgnite_Lightning_Hands.NS_HydroIgnite_Lightning_Hands" (CustomCharacter|GetSkeletalMeshForPartOnPlayerPawn) "None" "0, 0, 0" "0, 0, 0" "SnapToTarget" true)))

(event EventTick
  (params (DeltaSeconds :type Float%20%28single-precision%29))
  (bind _returnvalue (CustomCharacter|GetAssociatedPlayerPawn))
  (bind _currentweapon (Variables|/Script/FortniteGame.FortPawn|GetCurrentWeapon _returnvalue))
  (|CallParentFunction DeltaSeconds)
  (Variables|Default|SetIs%20Performing%20Action (Weapon|IsWeaponHiddenbyHolstering _returnvalue 0))
  (Utilities|IsValid _currentweapon
    (:Is%20Valid
      (bind _output_get (Variables|Default|SetWeaponType (Item|GetWeaponDataItemType _currentweapon)))
      (if (and (== (Math|Conversions|ToInteger(Byte) _output_get) 23) (Variables|Default|GetPickaxeCheck))
        (if (not (Variables|Default|GetIs%20Performing%20Action))
          (if (Reactive|Frontend|IsCosmeticObjectDisplayedinFrontendContext 0)
            (Effects|Components|ParticleSystem|SetFloatParameter (CallFunction|GetIdleVFXComponent) "Pickaxe" 1.0)
            (Rendering|Material|SetScalarParameterValue (Variables|Default|GetLightningDMI) "Opacity" 1.0 :!class /Script/Engine.MaterialInstanceDynamic)
            (else
              (Effects|Components|ParticleSystem|SetFloatParameter (CallFunction|GetIdleVFXComponent) "Pickaxe" 1.0)
              (Rendering|Material|SetScalarParameterValue (Variables|Default|GetLightningDMI) "Opacity" 1.0 :!class /Script/Engine.MaterialInstanceDynamic)))
          (else
            (Effects|Components|ParticleSystem|SetFloatParameter (CallFunction|GetIdleVFXComponent) "Pickaxe" 0.0)
            (Rendering|Material|SetScalarParameterValue (Variables|Default|GetLightningDMI) "Opacity" 0.0 :!class /Script/Engine.MaterialInstanceDynamic)))
        (else
          _)))
    (:Is%20Not%20Valid
      _))
  (Effects|Components|ParticleSystem|SetFloatParameter (CallFunction|GetIdleVFXComponent) "Pickaxe" 0.0)
  (Rendering|Material|SetScalarParameterValue (Variables|Default|GetLightningDMI) "Opacity" 0.0 :!class /Script/Engine.MaterialInstanceDynamic))

(fn ConstructionScript)
