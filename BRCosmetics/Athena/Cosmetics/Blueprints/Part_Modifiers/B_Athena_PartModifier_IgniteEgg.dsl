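; Character part modifier for the IgniteEgg cosmetic that drives a hologram effect on the Head, Body and Face meshes.
; Customization seeds per-character random glitch values, and UpdateHologram pushes game time, hologram and DBNO parameters to materials and overlays on spawn and DBNO changes.
; DBNO also adds a Niagara component, deactivated after a delay on recovery.
; Damage-stat changes write a timestamp and mesh-local hit vector, using a fixed fallback for distant hits.
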(event EventOnCharacterCustomizationFinished
  (params (CharacterPartInfo :type Array%20of%20Applied%20Character%20Part%20Info%20Structures))
  (bind _randomnumx (Variables|Default|GetRandomNumX))
  (bind _randomnumy (Variables|Default|GetRandomNumY))
  ; Get all parts
  (|CallParentFunction CharacterPartInfo)
  (Variables|Default|SetMeshes (Utilities|Array|MakeArray (Customization|GetSkeletalMeshforPartType (CustomCharacter|GetAssociatedPlayerPawn) "Head" :!class /Script/FortniteGame.FortPlayerPawn) (Customization|GetSkeletalMeshforPartType (CustomCharacter|GetAssociatedPlayerPawn) "Body" :!class /Script/FortniteGame.FortPlayerPawn) (Customization|GetSkeletalMeshforPartType (CustomCharacter|GetAssociatedPlayerPawn) "Face" :!class /Script/FortniteGame.FortPlayerPawn)))
  ; Add per character random vec2 to hologram glitch
  (Variables|Default|SetRandomNumX (Math|Random|RandomFloat))
  (Variables|Default|SetRandomNumY (Math|Random|RandomFloat))
  (for _array_element (Variables|Default|GetMeshes)
    (Rendering|Material|SetScalarParameterValueonMaterials _array_element "IE_RandomX" _randomnumx)
    (Class|FortCosmeticMaterialBlueprintLibrary|SetScalarParameterValueonOverlayMaterial _array_element "IE_RandomX" _randomnumx)
    (Rendering|Material|SetScalarParameterValueonMaterials _array_element "IE_RandomY" _randomnumy)
    (Class|FortCosmeticMaterialBlueprintLibrary|SetScalarParameterValueonOverlayMaterial _array_element "IE_RandomY" _randomnumy)))

(event EventOnPostResIn
  (bind _returnvalue (CustomCharacter|GetSkeletalMeshForPartOnPlayerPawn))
  (bind _returnvalue_2 (Transformation|GetWorldTransform _returnvalue))
  ; DBNO
  (|CallParentFunction)
  (EventDispatchers|BindEventtoOnDBNOStateChanged (CustomCharacter|GetAssociatedPlayerPawn) (delegate Custom|DBNO))
  ; On Hit
  (Cosmetics|WatchCosmeticStat (CustomCharacter|GetAssociatedPlayerPawn) "(TagName=\"StatManager.DamageReceived\")" 0 false
    (:OnStatChanged
      (for _array_element (Variables|Default|GetMeshes)
        (if (>= (Math|Vector|VectorLength (Math|Transform|InverseTransformLocation _returnvalue_2 (.Location (Collision|BreakHitResult (stale EventOnHitPawn Hit))))) 155.0)
          (Rendering|Material|SetVectorParameterValueonMaterials _array_element "HitVector" "0.000000,0.000000,100.000000")
          (else
            (Rendering|Material|SetVectorParameterValueonMaterials _array_element "HitVector" (Math|Transform|InverseTransformLocation _returnvalue_2 (.Location (Collision|BreakHitResult (stale EventOnHitPawn Hit)))))))
        (Rendering|Material|SetScalarParameterValueonMaterials _array_element "IE_GameTimeFromBP_Hit" (Utilities|Time|GetGameTimeinSeconds))))))

(event Custom|DBNO
  (params (FortPawn :type Fort%20Pawn%20Object%20Reference) (bInIsDBNO :type Boolean))
  (bind _returnvalue (not bInIsDBNO))
  ; DBNO
  (CallFunction|UpdateHologram (Math|Conversions|ToFloat(Boolean) _returnvalue) 1.0)
  (if _returnvalue
    (Utilities|FlowControl|Delay 1.75)
    (for _array_element (Variables|Default|GetMeshes)
      (Rendering|Material|SetScalarParameterValueonMaterials _array_element "IE_IsDBNO" 0.0)
      (Class|FortCosmeticMaterialBlueprintLibrary|SetScalarParameterValueonOverlayMaterial _array_element "IE_IsDBNO")
      ; VFX
      (Niagara|SetNiagaraVariable(Float) (stale _returnvalue_2) "HologramColorLerp" 1.0)
      (Components|Activation|Deactivate (stale _returnvalue_2)))
    (else
      (bind _returnvalue_2 (AddComponent|NODE_AddNiagaraComponent-2 :self (CustomCharacter|GetAssociatedPlayerPawn)))
      (Niagara|SetNiagaraVariable(Float) _returnvalue_2 "HologramColorLerp" 0.0))))

(event EventOnHitPawn
  (params (SelfActor :type Actor%20Object%20Reference) (OtherActor :type Actor%20Object%20Reference) (NormalImpulse :type Vector :default "0, 0, 0") (Hit :type Hit%20Result%20Structure%20%28by%20ref%29))
  ; On Hit
  (|CallParentFunction SelfActor OtherActor NormalImpulse Hit))

(event EventOnPawnSet
  ; Hologram stablizing when character spawned
  (|CallParentFunction)
  (Utilities|FlowControl|Delay 0.0)
  (CallFunction|UpdateHologram 1.0))

(fn UpdateHologram
  (params (Hologram :type Float%20%28double-precision%29) (IsDBNO :type Float%20%28double-precision%29))
  (bind _returnvalue (Utilities|Time|GetGameTimeinSeconds))
  (bind _isdbno (Variables|GetIsDBNO))
  (bind _hologram (Variables|GetHologram))
  ; Update material and overlay
  (for _array_element (Variables|Default|GetMeshes)
    (Rendering|Material|SetScalarParameterValueonMaterials _array_element "IE_GameTimeFromBP" _returnvalue)
    (Rendering|Material|SetScalarParameterValueonMaterials _array_element "IE_HologramFX" _hologram)
    (Rendering|Material|SetScalarParameterValueonMaterials _array_element "IE_IsDBNO" _isdbno)
    (Class|FortCosmeticMaterialBlueprintLibrary|SetScalarParameterValueonOverlayMaterial _array_element "IE_GameTimeFromBP" _returnvalue)
    (Class|FortCosmeticMaterialBlueprintLibrary|SetScalarParameterValueonOverlayMaterial _array_element "IE_HologramFX" _hologram)
    (Class|FortCosmeticMaterialBlueprintLibrary|SetScalarParameterValueonOverlayMaterial _array_element "IE_IsDBNO" _isdbno)))

(fn ConstructionScript
  (|CallParentFunction))
