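; Cosmetic part modifier for the IgniteEgg back bling that makes its effects react to the owning player's eliminations.
; After post-res-in, it watches the kill stat, or fakes kills on a looping timer when frontend reactivity should display.
; Each kill stamps game time into the idle VFX ReactivityTime parameter if the mesh is visible.
; UpdateHologram pushes game time, hologram and DBNO values to mesh and overlay materials.
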(event EventOnPostResIn
  (|CallParentFunction)
  (if (Reactive|Frontend|ShouldDisplayFrontendReactivity 0 "Backpack")
    ; Frontend Loop
    (Utilities|FlowControl|Delay 5.0)
    (CallFunction|OnKill)
    (Utilities|Time|SetTimerbyEvent (delegate Custom|OnKill) 8.0 true)
    (else
      ; Setup On Kill
      (bind (_statname _statvalue) (Cosmetics|WatchCosmeticStat (CustomCharacter|GetAssociatedPlayerPawn) "(TagName=\"StatManager.AthenaKills\")" 0 false)
        (:OnStatChanged
          (if (> _statvalue 0)
            (CallFunction|OnKill)))))))

(event Custom|OnKill
  ; On Kill Reactivity
  (if (Rendering|IsVisible (CustomCharacter|GetSkeletalMeshForPartOnPlayerPawn))
    (Effects|Components|ParticleSystem|SetFloatParameter (CallFunction|GetIdleVFXComponent) "ReactivityTime" (Utilities|Time|GetGameTimeinSeconds))))

(fn UpdateHologram
  (params (Hologram :type Float%20%28double-precision%29) (IsDBNO :type Float%20%28double-precision%29))
  (bind _returnvalue (Utilities|Time|GetGameTimeinSeconds))
  (bind _isdbno (Variables|GetIsDBNO))
  (bind _hologram (Variables|GetHologram))
  ; Update material and overlay
  (for _array_element (Variables|Default|GetMeshes)
    (Rendering|Material|SetScalarParameterValueonMaterials _array_element "IE_GameTimeFromBP" _returnvalue)
    (Rendering|Material|SetScalarParameterValueonMaterials _array_element "IE_HologramFX" _hologram)
    (Rendering|Material|SetScalarParameterValueonMaterials _array_element "IE_IsDBNO" _isdbno)
    (Class|FortCosmeticMaterialBlueprintLibrary|SetScalarParameterValueonOverlayMaterial _array_element "IE_GameTimeFromBP" _returnvalue)
    (Class|FortCosmeticMaterialBlueprintLibrary|SetScalarParameterValueonOverlayMaterial _array_element "IE_HologramFX" _hologram)
    (Class|FortCosmeticMaterialBlueprintLibrary|SetScalarParameterValueonOverlayMaterial _array_element "IE_IsDBNO" _isdbno)))

(fn ConstructionScript
  (|CallParentFunction))
