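; Part modifier Blueprint for the Inferno FNCS character skin that configures its flame and back-hair idle Niagara VFX.
; After res-in it assigns hair meshes and a root-position texture, then stamps BurstTime with game time: after a short delay in frontend, otherwise when the watched kill stat changes above zero.
; SetupSuperLevel selects one of three material maps by index, pushes each entry into Niagara, and sets the super-level flag and index.
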(event Custom|SetupSuperLevel
  (params (SuperLevelIndex :type Integer))
  (bind _returnvalue (CallFunction|GetIdleVFXComponent))
  (bind _asniagara_particle_system_component (Utilities|Casting|CastToNiagaraComponent _returnvalue))
  (bind _superlevel_rmaterials (Variables|Default|GetSuperLevel_RMaterials))
  (bind _superlevel_gmaterials (Variables|Default|GetSuperLevel_GMaterials))
  (bind _superlevel_ppmaterials (Variables|Default|GetSuperLevel_PPMaterials))
  (bind _returnvalue_2 (Utilities|Select :Index SuperLevelIndex :Option%200 _superlevel_ppmaterials :Option%201 _superlevel_gmaterials :Option%202 _superlevel_rmaterials))
  (bind _keys (Utilities|Map|Keys _returnvalue_2))
  (bind _values (Utilities|Map|Values _returnvalue_2))
  (for (_array_element _array_index) _keys
    (Niagara|SetNiagaraVariable(Material) _asniagara_particle_system_component _array_element (Utilities|Array|Get(acopy) _values _array_index)))
  (Niagara|SetNiagaraVariable(Bool) _asniagara_particle_system_component "bIsSL" true)
  (Niagara|SetNiagaraVariable(Int32) _asniagara_particle_system_component "SuperLevelIndex" SuperLevelIndex))

(event EventOnPostResIn
  (bind _returnvalue (CallFunction|GetIdleVFXComponent))
  (bind _returnvalue_2 (CallFunction|GetIdleVFXComponent))
  (bind _asniagara_particle_system_component (Utilities|Casting|CastToNiagaraComponent _returnvalue))
  (bind _asniagara_particle_system_component_2 (Utilities|Casting|CastToNiagaraComponent _returnvalue_2))
  ;[1 0 0.02787 0.2 200] FX
  (|CallParentFunction)
  (if (Reactive|Frontend|IsCosmeticObjectDisplayedinFrontendContext 0)
    (Utilities|FlowControl|Sequence
      (:then_0
        (Utilities|FlowControl|Delay 0.3)
        (if (Reactive|Frontend|IsCosmeticObjectDisplayedinFrontendContext 0)
          (Effects|Components|ParticleSystem|SetFloatParameter (CallFunction|GetIdleVFXComponent) "BurstTime" (Utilities|Time|GetGameTimeinSeconds))
          (Utilities|FlowControl|Delay 1e+01)))
      (:then_1
        (Niagara|SetTextureObject _asniagara_particle_system_component "BackHairRootPos" "/BRCosmetics/Characters/Player/Male/Medium/Bodies/M_MED_Inferno/FX/Mat/T_BackHairPos.T_BackHairPos")
        (Niagara|SetNiagaraVariable(StaticMesh) _asniagara_particle_system_component "SM_Flame" "/Game/Tools/SplineMeshGenerator/Saves/Meshes/SM_Character_Inferno_BackHair.SM_Character_Inferno_BackHair")
        (Niagara|SetNiagaraVariable(StaticMesh) _asniagara_particle_system_component "SM_BackHair" "/Game/Tools/SplineMeshGenerator/Saves/Meshes/SM_Character_Inferno_BackHair2.SM_Character_Inferno_BackHair2")))
    (else
      (Utilities|FlowControl|Sequence
        (:then_0
          (bind (_statname _statvalue) (Cosmetics|WatchCosmeticStat (CustomCharacter|GetAssociatedPlayerPawn) "(TagName=\"StatManager.AthenaKills\")")
            (:OnStatChanged
              (if (> _statvalue 0)
                (Effects|Components|ParticleSystem|SetFloatParameter (CallFunction|GetIdleVFXComponent) "BurstTime" (Utilities|Time|GetGameTimeinSeconds))))))
        (:then_1
          (Niagara|SetTextureObject _asniagara_particle_system_component_2 "BackHairRootPos" "/BRCosmetics/Characters/Player/Male/Medium/Bodies/M_MED_Inferno/FX/Mat/T_BackHairPos.T_BackHairPos")
          (Niagara|SetNiagaraVariable(StaticMesh) _asniagara_particle_system_component_2 "SM_Flame" "/Game/Tools/SplineMeshGenerator/Saves/Meshes/SM_Character_Inferno_BackHair.SM_Character_Inferno_BackHair")
          (Niagara|SetNiagaraVariable(StaticMesh) _asniagara_particle_system_component_2 "SM_BackHair" "/Game/Tools/SplineMeshGenerator/Saves/Meshes/SM_Character_Inferno_BackHair2.SM_Character_Inferno_BackHair2"))))))

(fn ConstructionScript
  (|CallParentFunction))
