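; Character part modifier for the Inferno skin that configures its idle Niagara flame-hair VFX after res-in.
; Frontend infers the style from mesh slot 0's parent material; in-game reads the Material variant channel.
; Styles select normal, white or Super Level materials (via SetupSuperLevel's maps), plus shared hair meshes and texture.
; With the first Emissive variant, BurstTime fires after a frontend delay or on in-game kill-stat changes.
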(event Custom|SetupSuperLevel
  (params (SuperLevelIndex :type Integer))
  (bind _returnvalue (CallFunction|GetIdleVFXComponent))
  (bind _asniagara_particle_system_component (Utilities|Casting|CastToNiagaraComponent _returnvalue))
  (bind _superlevel_rmaterials (Variables|Default|GetSuperLevel_RMaterials))
  (bind _superlevel_gmaterials (Variables|Default|GetSuperLevel_GMaterials))
  (bind _superlevel_ppmaterials (Variables|Default|GetSuperLevel_PPMaterials))
  (bind _returnvalue_2 (Utilities|Select :Index SuperLevelIndex :Option%200 _superlevel_ppmaterials :Option%201 _superlevel_gmaterials :Option%202 _superlevel_rmaterials))
  (bind _keys (Utilities|Map|Keys _returnvalue_2))
  (bind _values (Utilities|Map|Values _returnvalue_2))
  (for (_array_element _array_index) _keys
    (Niagara|SetNiagaraVariable(Material) _asniagara_particle_system_component _array_element (Utilities|Array|Get(acopy) _values _array_index)))
  (Niagara|SetNiagaraVariable(Bool) _asniagara_particle_system_component "bIsSL" true)
  (Niagara|SetNiagaraVariable(Int32) _asniagara_particle_system_component "SuperLevelIndex" SuperLevelIndex))

(event EventOnPostResIn
  (bind _returnvalue (CallFunction|GetIdleVFXComponent))
  (bind _returnvalue_2 (CustomCharacter|GetSkeletalMeshForPartOnPlayerPawn))
  (bind _returnvalue_3 (CallFunction|GetIdleVFXComponent))
  (bind _returnvalue_4 (CallFunction|GetIdleVFXComponent))
  (bind _returnvalue_5 (CallFunction|GetIdleVFXComponent))
  (bind _asniagara_particle_system_component (Utilities|Casting|CastToNiagaraComponent _returnvalue_3))
  (bind _asniagara_particle_system_component_2 (Utilities|Casting|CastToNiagaraComponent _returnvalue))
  (bind _asniagara_particle_system_component_3 (Utilities|Casting|CastToNiagaraComponent _returnvalue_4))
  (bind _asniagara_particle_system_component_4 (Utilities|Casting|CastToNiagaraComponent _returnvalue_5))
  (|CallParentFunction)
  ;[1 0 0.02787 0.2 200] FX
  (if (Reactive|Frontend|IsCosmeticObjectDisplayedinFrontendContext 0)
    (Utilities|FlowControl|Sequence
      (:then_0
        (bind _index (Class|FortCosmeticFunctionLibrary|GetVariantChannelActiveIndexForSlot "(TagName=\"LoadoutSlot.Character\")" (CosmeticLoadout|GetCosmeticLoadout) "(TagName=\"Cosmetics.Variant.Channel.Emissive\")" (Utilities|Array|MakeArray "(TagName=\"Cosmetics.Variant.Property.Emissive1\")" "(TagName=\"Cosmetics.Variant.Property.Emissive2\")")))
        (if (== _index 0)
          (Utilities|FlowControl|Delay 0.3)
          (if (Reactive|Frontend|IsCosmeticObjectDisplayedinFrontendContext 0)
            (Effects|Components|ParticleSystem|SetFloatParameter (CallFunction|GetIdleVFXComponent) "BurstTime" (Utilities|Time|GetGameTimeinSeconds))
            (Utilities|FlowControl|Delay 1e+01))))
      (:then_1
        (bind _asmaterial_instance (Utilities|Casting|CastToMaterialInstance (Rendering|Material|GetMaterial (CustomCharacter|GetSkeletalMeshForPartOnPlayerPawn) 0))
          (:then
            (Development|PrintString (Utilities|GetDisplayName (Variables|/Script/Engine.MaterialInstance|GetParent _asmaterial_instance)))
            (Utilities|FlowControl|Sequence
              (:then_0
                (if (Math|Boolean|ORBoolean (Math|Boolean|ORBoolean (== (Variables|/Script/Engine.MaterialInstance|GetParent _asmaterial_instance) "/BRCosmetics/Characters/Player/Male/Medium/Bodies/M_MED_Inferno/Skins/Super_Levels/Materials/M_S24_Inferno_Body_SL1.M_S24_Inferno_Body_SL1") (== (Variables|/Script/Engine.MaterialInstance|GetParent _asmaterial_instance) "/BRCosmetics/Characters/Player/Male/Medium/Bodies/M_MED_Inferno/Skins/Super_Levels/Materials/M_S24_InfernoCoat_Body_SL1.M_S24_InfernoCoat_Body_SL1") (== (Variables|/Script/Engine.MaterialInstance|GetParent _asmaterial_instance) "/BRCosmetics/Characters/Player/Male/Medium/Bodies/M_MED_Inferno/Skins/Super_Levels/Materials/M_S24_Inferno_FaceAcc_SL1.M_S24_Inferno_FaceAcc_SL1")) (Math|Boolean|ORBoolean (== (Variables|/Script/Engine.MaterialInstance|GetParent _asmaterial_instance) "/BRCosmetics/Characters/Player/Male/Medium/Bodies/M_MED_Inferno/Skins/Super_Levels/Materials/M_S24_Inferno_Body_SL2.M_S24_Inferno_Body_SL2") (== (Variables|/Script/Engine.MaterialInstance|GetParent _asmaterial_instance) "/BRCosmetics/Characters/Player/Male/Medium/Bodies/M_MED_Inferno/Skins/Super_Levels/Materials/M_S24_InfernoCoat_Body_SL2.M_S24_InfernoCoat_Body_SL2") (== (Variables|/Script/Engine.MaterialInstance|GetParent _asmaterial_instance) "/BRCosmetics/Characters/Player/Male/Medium/Bodies/M_MED_Inferno/Skins/Super_Levels/Materials/M_S24_Inferno_FaceAcc_SL2.M_S24_Inferno_FaceAcc_SL2")) (Math|Boolean|ORBoolean (== (Variables|/Script/Engine.MaterialInstance|GetParent _asmaterial_instance) "/BRCosmetics/Characters/Player/Male/Medium/Bodies/M_MED_Inferno/Skins/Super_Levels/Materials/M_S24_Inferno_Body_SL3.M_S24_Inferno_Body_SL3") (== (Variables|/Script/Engine.MaterialInstance|GetParent _asmaterial_instance) "/BRCosmetics/Characters/Player/Male/Medium/Bodies/M_MED_Inferno/Skins/Super_Levels/Materials/M_S24_InfernoCoat_Body_SL3.M_S24_InfernoCoat_Body_SL3") (== (Variables|/Script/Engine.MaterialInstance|GetParent _asmaterial_instance) "/BRCosmetics/Characters/Player/Male/Medium/Bodies/M_MED_Inferno/Skins/Super_Levels/Materials/M_S24_Inferno_FaceAcc_SL3.M_S24_Inferno_FaceAcc_SL3")))
                  (Utilities|FlowControl|Sequence
                    (:then_0
                      (if (Math|Boolean|ORBoolean (== (Variables|/Script/Engine.MaterialInstance|GetParent _asmaterial_instance) "/BRCosmetics/Characters/Player/Male/Medium/Bodies/M_MED_Inferno/Skins/Super_Levels/Materials/M_S24_Inferno_Body_SL1.M_S24_Inferno_Body_SL1") (== (Variables|/Script/Engine.MaterialInstance|GetParent _asmaterial_instance) "/BRCosmetics/Characters/Player/Male/Medium/Bodies/M_MED_Inferno/Skins/Super_Levels/Materials/M_S24_InfernoCoat_Body_SL1.M_S24_InfernoCoat_Body_SL1") (== (Variables|/Script/Engine.MaterialInstance|GetParent _asmaterial_instance) "/BRCosmetics/Characters/Player/Male/Medium/Bodies/M_MED_Inferno/Skins/Super_Levels/Materials/M_S24_Inferno_FaceAcc_SL1.M_S24_Inferno_FaceAcc_SL1"))
                        (CallFunction|SetupSuperLevel 0)))
                    (:then_1
                      (if (Math|Boolean|ORBoolean (== (Variables|/Script/Engine.MaterialInstance|GetParent _asmaterial_instance) "/BRCosmetics/Characters/Player/Male/Medium/Bodies/M_MED_Inferno/Skins/Super_Levels/Materials/M_S24_Inferno_Body_SL2.M_S24_Inferno_Body_SL2") (== (Variables|/Script/Engine.MaterialInstance|GetParent _asmaterial_instance) "/BRCosmetics/Characters/Player/Male/Medium/Bodies/M_MED_Inferno/Skins/Super_Levels/Materials/M_S24_InfernoCoat_Body_SL2.M_S24_InfernoCoat_Body_SL2") (== (Variables|/Script/Engine.MaterialInstance|GetParent _asmaterial_instance) "/BRCosmetics/Characters/Player/Male/Medium/Bodies/M_MED_Inferno/Skins/Super_Levels/Materials/M_S24_Inferno_FaceAcc_SL2.M_S24_Inferno_FaceAcc_SL2"))
                        (CallFunction|SetupSuperLevel 1)))
                    (:then_2
                      (if (Math|Boolean|ORBoolean (== (Variables|/Script/Engine.MaterialInstance|GetParent _asmaterial_instance) "/BRCosmetics/Characters/Player/Male/Medium/Bodies/M_MED_Inferno/Skins/Super_Levels/Materials/M_S24_Inferno_Body_SL3.M_S24_Inferno_Body_SL3") (== (Variables|/Script/Engine.MaterialInstance|GetParent _asmaterial_instance) "/BRCosmetics/Characters/Player/Male/Medium/Bodies/M_MED_Inferno/Skins/Super_Levels/Materials/M_S24_InfernoCoat_Body_SL3.M_S24_InfernoCoat_Body_SL3") (== (Variables|/Script/Engine.MaterialInstance|GetParent _asmaterial_instance) "/BRCosmetics/Characters/Player/Male/Medium/Bodies/M_MED_Inferno/Skins/Super_Levels/Materials/M_S24_Inferno_FaceAcc_SL3.M_S24_Inferno_FaceAcc_SL3"))
                        (CallFunction|SetupSuperLevel 2))))
                  (else
                    (Niagara|SetNiagaraVariable(Material) _asniagara_particle_system_component_4 "Mat_BackFlame" (Selection|SelectMaterialInstance "/BRCosmetics/Characters/Player/Male/Medium/Bodies/M_MED_Inferno/FX/Mat/MI_Inferno_FlameHair_Back.MI_Inferno_FlameHair_Back" "/BRCosmetics/Characters/Player/Male/Medium/Bodies/M_MED_Inferno/FX/Mat/MI_Inferno_FlameHair_Back_White.MI_Inferno_FlameHair_Back_White" (Math|Boolean|ORBoolean (== (Variables|/Script/Engine.MaterialInstance|GetParent _asmaterial_instance) "/BRCosmetics/Characters/Player/Male/Medium/Bodies/M_MED_Inferno/Materials/M_Med_Inferno_Body.M_Med_Inferno_Body") (== (Variables|/Script/Engine.MaterialInstance|GetParent _asmaterial_instance) "/BRCosmetics/Characters/Player/Male/Medium/Bodies/M_MED_Inferno/Materials/M_Med_Inferno_Coat.M_Med_Inferno_Coat") (== (Variables|/Script/Engine.MaterialInstance|GetParent _asmaterial_instance) "/BRCosmetics/Characters/Player/Male/Medium/Bodies/M_MED_Inferno/Materials/M_Med_Inferno_FaceAcc.M_Med_Inferno_FaceAcc"))))
                    (Niagara|SetNiagaraVariable(Material) _asniagara_particle_system_component_4 "Mat_BackHair" (Selection|SelectMaterialInstance "/BRCosmetics/Characters/Player/Male/Medium/Bodies/M_MED_Inferno/FX/Mat/MI_Inferno_MainHair_BackHair.MI_Inferno_MainHair_BackHair" "/BRCosmetics/Characters/Player/Male/Medium/Bodies/M_MED_Inferno/FX/Mat/MI_Inferno_MainHair_BackHair_White_NoColourful.MI_Inferno_MainHair_BackHair_White_NoColourful" (Math|Boolean|ORBoolean (== (Variables|/Script/Engine.MaterialInstance|GetParent _asmaterial_instance) "/BRCosmetics/Characters/Player/Male/Medium/Bodies/M_MED_Inferno/Materials/M_Med_Inferno_Body.M_Med_Inferno_Body") (== (Variables|/Script/Engine.MaterialInstance|GetParent _asmaterial_instance) "/BRCosmetics/Characters/Player/Male/Medium/Bodies/M_MED_Inferno/Materials/M_Med_Inferno_Coat.M_Med_Inferno_Coat") (== (Variables|/Script/Engine.MaterialInstance|GetParent _asmaterial_instance) "/BRCosmetics/Characters/Player/Male/Medium/Bodies/M_MED_Inferno/Materials/M_Med_Inferno_FaceAcc.M_Med_Inferno_FaceAcc"))))
                    (Niagara|SetNiagaraVariable(Material) _asniagara_particle_system_component_4 "Mat_EyeSpark" (Selection|SelectMaterialInstance "/BRCosmetics/Characters/Player/Male/Medium/Bodies/M_MED_Inferno/FX/Mat/M_Inferno_CrossHair_ForUse.M_Inferno_CrossHair_ForUse" "/BRCosmetics/Characters/Player/Male/Medium/Bodies/M_MED_Inferno/FX/Mat/M_Inferno_CrossHair_White.M_Inferno_CrossHair_White" (Math|Boolean|ORBoolean (== (Variables|/Script/Engine.MaterialInstance|GetParent _asmaterial_instance) "/BRCosmetics/Characters/Player/Male/Medium/Bodies/M_MED_Inferno/Materials/M_Med_Inferno_Body.M_Med_Inferno_Body") (== (Variables|/Script/Engine.MaterialInstance|GetParent _asmaterial_instance) "/BRCosmetics/Characters/Player/Male/Medium/Bodies/M_MED_Inferno/Materials/M_Med_Inferno_Coat.M_Med_Inferno_Coat") (== (Variables|/Script/Engine.MaterialInstance|GetParent _asmaterial_instance) "/BRCosmetics/Characters/Player/Male/Medium/Bodies/M_MED_Inferno/Materials/M_Med_Inferno_FaceAcc.M_Med_Inferno_FaceAcc"))))
                    (Niagara|SetNiagaraVariable(Material) _asniagara_particle_system_component_4 "Mat_FstLayer" (Selection|SelectMaterialInstance "/BRCosmetics/Characters/Player/Male/Medium/Bodies/M_MED_Inferno/FX/Mat/MI_Inferno_MainHair_FstLayer.MI_Inferno_MainHair_FstLayer" "/BRCosmetics/Characters/Player/Male/Medium/Bodies/M_MED_Inferno/FX/Mat/MI_Inferno_MainHair_FstLayer_White.MI_Inferno_MainHair_FstLayer_White" (Math|Boolean|ORBoolean (== (Variables|/Script/Engine.MaterialInstance|GetParent _asmaterial_instance) "/BRCosmetics/Characters/Player/Male/Medium/Bodies/M_MED_Inferno/Materials/M_Med_Inferno_Body.M_Med_Inferno_Body") (== (Variables|/Script/Engine.MaterialInstance|GetParent _asmaterial_instance) "/BRCosmetics/Characters/Player/Male/Medium/Bodies/M_MED_Inferno/Materials/M_Med_Inferno_Coat.M_Med_Inferno_Coat") (== (Variables|/Script/Engine.MaterialInstance|GetParent _asmaterial_instance) "/BRCosmetics/Characters/Player/Male/Medium/Bodies/M_MED_Inferno/Materials/M_Med_Inferno_FaceAcc.M_Med_Inferno_FaceAcc"))))
                    (Niagara|SetNiagaraVariable(Material) _asniagara_particle_system_component_4 "Mat_HighLayer" (Selection|SelectMaterialInstance "/BRCosmetics/Characters/Player/Male/Medium/Bodies/M_MED_Inferno/FX/Mat/MI_Inferno_MainHair_HighHair.MI_Inferno_MainHair_HighHair" "/BRCosmetics/Characters/Player/Male/Medium/Bodies/M_MED_Inferno/FX/Mat/MI_Inferno_MainHair_HighHair_White.MI_Inferno_MainHair_HighHair_White" (Math|Boolean|ORBoolean (== (Variables|/Script/Engine.MaterialInstance|GetParent _asmaterial_instance) "/BRCosmetics/Characters/Player/Male/Medium/Bodies/M_MED_Inferno/Materials/M_Med_Inferno_Body.M_Med_Inferno_Body") (== (Variables|/Script/Engine.MaterialInstance|GetParent _asmaterial_instance) "/BRCosmetics/Characters/Player/Male/Medium/Bodies/M_MED_Inferno/Materials/M_Med_Inferno_Coat.M_Med_Inferno_Coat") (== (Variables|/Script/Engine.MaterialInstance|GetParent _asmaterial_instance) "/BRCosmetics/Characters/Player/Male/Medium/Bodies/M_MED_Inferno/Materials/M_Med_Inferno_FaceAcc.M_Med_Inferno_FaceAcc"))))
                    (Niagara|SetNiagaraVariable(Material) _asniagara_particle_system_component_4 "Mat_MainFlame" (Selection|SelectMaterialInstance "/BRCosmetics/Characters/Player/Male/Medium/Bodies/M_MED_Inferno/FX/Mat/M_Inferno_FlameHair_ForUse.M_Inferno_FlameHair_ForUse" "/BRCosmetics/Characters/Player/Male/Medium/Bodies/M_MED_Inferno/FX/Mat/M_Inferno_FlameHair_White.M_Inferno_FlameHair_White" (Math|Boolean|ORBoolean (== (Variables|/Script/Engine.MaterialInstance|GetParent _asmaterial_instance) "/BRCosmetics/Characters/Player/Male/Medium/Bodies/M_MED_Inferno/Materials/M_Med_Inferno_Body.M_Med_Inferno_Body") (== (Variables|/Script/Engine.MaterialInstance|GetParent _asmaterial_instance) "/BRCosmetics/Characters/Player/Male/Medium/Bodies/M_MED_Inferno/Materials/M_Med_Inferno_Coat.M_Med_Inferno_Coat") (== (Variables|/Script/Engine.MaterialInstance|GetParent _asmaterial_instance) "/BRCosmetics/Characters/Player/Male/Medium/Bodies/M_MED_Inferno/Materials/M_Med_Inferno_FaceAcc.M_Med_Inferno_FaceAcc"))))
                    (Niagara|SetNiagaraVariable(Material) _asniagara_particle_system_component_4 "Mat_SecLayer" (Selection|SelectMaterialInstance "/BRCosmetics/Characters/Player/Male/Medium/Bodies/M_MED_Inferno/FX/Mat/MI_Inferno_MainHair_SecLayer.MI_Inferno_MainHair_SecLayer" "/BRCosmetics/Characters/Player/Male/Medium/Bodies/M_MED_Inferno/FX/Mat/MI_Inferno_MainHair_SecLayer_White.MI_Inferno_MainHair_SecLayer_White" (Math|Boolean|ORBoolean (== (Variables|/Script/Engine.MaterialInstance|GetParent _asmaterial_instance) "/BRCosmetics/Characters/Player/Male/Medium/Bodies/M_MED_Inferno/Materials/M_Med_Inferno_Body.M_Med_Inferno_Body") (== (Variables|/Script/Engine.MaterialInstance|GetParent _asmaterial_instance) "/BRCosmetics/Characters/Player/Male/Medium/Bodies/M_MED_Inferno/Materials/M_Med_Inferno_Coat.M_Med_Inferno_Coat") (== (Variables|/Script/Engine.MaterialInstance|GetParent _asmaterial_instance) "/BRCosmetics/Characters/Player/Male/Medium/Bodies/M_MED_Inferno/Materials/M_Med_Inferno_FaceAcc.M_Med_Inferno_FaceAcc"))))
                    (Niagara|SetNiagaraVariable(Material) _asniagara_particle_system_component_4 "Mat_TempEmissiveArea" (Selection|SelectMaterialInstance "/BRCosmetics/Characters/Player/Male/Medium/Bodies/M_MED_Inferno/FX/Mat/M_Inferno_TempCenterEmissive_ForUse.M_Inferno_TempCenterEmissive_ForUse" "/BRCosmetics/Characters/Player/Male/Medium/Bodies/M_MED_Inferno/FX/Mat/M_Inferno_TempCenterEmissive_White.M_Inferno_TempCenterEmissive_White" (Math|Boolean|ORBoolean (== (Variables|/Script/Engine.MaterialInstance|GetParent _asmaterial_instance) "/BRCosmetics/Characters/Player/Male/Medium/Bodies/M_MED_Inferno/Materials/M_Med_Inferno_Body.M_Med_Inferno_Body") (== (Variables|/Script/Engine.MaterialInstance|GetParent _asmaterial_instance) "/BRCosmetics/Characters/Player/Male/Medium/Bodies/M_MED_Inferno/Materials/M_Med_Inferno_Coat.M_Med_Inferno_Coat") (== (Variables|/Script/Engine.MaterialInstance|GetParent _asmaterial_instance) "/BRCosmetics/Characters/Player/Male/Medium/Bodies/M_MED_Inferno/Materials/M_Med_Inferno_FaceAcc.M_Med_Inferno_FaceAcc"))))
                    (Niagara|SetNiagaraVariable(Material) _asniagara_particle_system_component_4 "Mat_TempEmissiveTranslucent" (Selection|SelectMaterialInstance "/BRCosmetics/Characters/Player/Male/Medium/Bodies/M_MED_Inferno/FX/Mat/MI_Inferno_TempCenterEmissive.MI_Inferno_TempCenterEmissive" "/BRCosmetics/Characters/Player/Male/Medium/Bodies/M_MED_Inferno/FX/Mat/MI_Inferno_TempCenterEmissive_White.MI_Inferno_TempCenterEmissive_White" (Math|Boolean|ORBoolean (== (Variables|/Script/Engine.MaterialInstance|GetParent _asmaterial_instance) "/BRCosmetics/Characters/Player/Male/Medium/Bodies/M_MED_Inferno/Materials/M_Med_Inferno_Body.M_Med_Inferno_Body") (== (Variables|/Script/Engine.MaterialInstance|GetParent _asmaterial_instance) "/BRCosmetics/Characters/Player/Male/Medium/Bodies/M_MED_Inferno/Materials/M_Med_Inferno_Coat.M_Med_Inferno_Coat") (== (Variables|/Script/Engine.MaterialInstance|GetParent _asmaterial_instance) "/BRCosmetics/Characters/Player/Male/Medium/Bodies/M_MED_Inferno/Materials/M_Med_Inferno_FaceAcc.M_Med_Inferno_FaceAcc"))))
                    (Niagara|SetNiagaraVariable(Material) _asniagara_particle_system_component_4 "Mat_TopFlame" (Selection|SelectMaterialInstance "/BRCosmetics/Characters/Player/Male/Medium/Bodies/M_MED_Inferno/FX/Mat/MI_Inferno_FlameHair_Top.MI_Inferno_FlameHair_Top" "/BRCosmetics/Characters/Player/Male/Medium/Bodies/M_MED_Inferno/FX/Mat/MI_Inferno_FlameHair_Top_White.MI_Inferno_FlameHair_Top_White" (Math|Boolean|ORBoolean (== (Variables|/Script/Engine.MaterialInstance|GetParent _asmaterial_instance) "/BRCosmetics/Characters/Player/Male/Medium/Bodies/M_MED_Inferno/Materials/M_Med_Inferno_Body.M_Med_Inferno_Body") (== (Variables|/Script/Engine.MaterialInstance|GetParent _asmaterial_instance) "/BRCosmetics/Characters/Player/Male/Medium/Bodies/M_MED_Inferno/Materials/M_Med_Inferno_Coat.M_Med_Inferno_Coat") (== (Variables|/Script/Engine.MaterialInstance|GetParent _asmaterial_instance) "/BRCosmetics/Characters/Player/Male/Medium/Bodies/M_MED_Inferno/Materials/M_Med_Inferno_FaceAcc.M_Med_Inferno_FaceAcc"))))
                    (Niagara|SetNiagaraVariable(Material) _asniagara_particle_system_component_4 "Mat_LegMat" (Selection|SelectMaterialInstance "/BRCosmetics/Characters/Player/Male/Medium/Bodies/M_MED_Inferno/FX/Mat/MI_Inferno_FlameHair_Leg.MI_Inferno_FlameHair_Leg" "/BRCosmetics/Characters/Player/Male/Medium/Bodies/M_MED_Inferno/FX/Mat/MI_Inferno_FlameHair_Leg_White.MI_Inferno_FlameHair_Leg_White" (Math|Boolean|ORBoolean (== (Variables|/Script/Engine.MaterialInstance|GetParent _asmaterial_instance) "/BRCosmetics/Characters/Player/Male/Medium/Bodies/M_MED_Inferno/Materials/M_Med_Inferno_Body.M_Med_Inferno_Body") (== (Variables|/Script/Engine.MaterialInstance|GetParent _asmaterial_instance) "/BRCosmetics/Characters/Player/Male/Medium/Bodies/M_MED_Inferno/Materials/M_Med_Inferno_Coat.M_Med_Inferno_Coat") (== (Variables|/Script/Engine.MaterialInstance|GetParent _asmaterial_instance) "/BRCosmetics/Characters/Player/Male/Medium/Bodies/M_MED_Inferno/Materials/M_Med_Inferno_FaceAcc.M_Med_Inferno_FaceAcc"))))
                    (Niagara|SetNiagaraVariable(Material) _asniagara_particle_system_component_4 "Mat_NewFlame" (Selection|SelectMaterialInstance "/BRCosmetics/Gadgets/Assets/VinderTech_GliderChute/Glider_Inferno/FX/MI_Glider_Inferno_MeshFlame_01.MI_Glider_Inferno_MeshFlame_01" "/BRCosmetics/Characters/Player/Male/Medium/Bodies/M_MED_Inferno/FX/Mat/GliderMaterial/MI_Glider_Inferno_MeshFlame_01_White.MI_Glider_Inferno_MeshFlame_01_White" (Math|Boolean|ORBoolean (== (Variables|/Script/Engine.MaterialInstance|GetParent _asmaterial_instance) "/BRCosmetics/Characters/Player/Male/Medium/Bodies/M_MED_Inferno/Materials/M_Med_Inferno_Body.M_Med_Inferno_Body") (== (Variables|/Script/Engine.MaterialInstance|GetParent _asmaterial_instance) "/BRCosmetics/Characters/Player/Male/Medium/Bodies/M_MED_Inferno/Materials/M_Med_Inferno_Coat.M_Med_Inferno_Coat") (== (Variables|/Script/Engine.MaterialInstance|GetParent _asmaterial_instance) "/BRCosmetics/Characters/Player/Male/Medium/Bodies/M_MED_Inferno/Materials/M_Med_Inferno_FaceAcc.M_Med_Inferno_FaceAcc"))))
                    (Niagara|SetNiagaraVariable(Bool) _asniagara_particle_system_component_4 "bIsSL" false))))
              (:then_1
                (Niagara|SetNiagaraVariable(Bool) _asniagara_particle_system_component_4 "bIsWhiteVersion" (not (Math|Boolean|ORBoolean (== (Variables|/Script/Engine.MaterialInstance|GetParent _asmaterial_instance) "/BRCosmetics/Characters/Player/Male/Medium/Bodies/M_MED_Inferno/Materials/M_Med_Inferno_Body.M_Med_Inferno_Body") (== (Variables|/Script/Engine.MaterialInstance|GetParent _asmaterial_instance) "/BRCosmetics/Characters/Player/Male/Medium/Bodies/M_MED_Inferno/Materials/M_Med_Inferno_Coat.M_Med_Inferno_Coat") (== (Variables|/Script/Engine.MaterialInstance|GetParent _asmaterial_instance) "/BRCosmetics/Characters/Player/Male/Medium/Bodies/M_MED_Inferno/Materials/M_Med_Inferno_FaceAcc.M_Med_Inferno_FaceAcc"))))
                ; Frontend Cloth Adjustments for InfernoFNCS - disabled for now
                (if (== (Variables|/Script/Engine.MaterialInstance|GetParent _asmaterial_instance) "/BRCosmetics/Characters/Player/Male/Medium/Bodies/M_MED_InfernoFNCS/Materials/MI_InfernoFNCS_Body.MI_InfernoFNCS_Body")
                  (Rendering|Material|SetScalarParameterValueonMaterials _returnvalue_2 "ClothRoughness" 2.0) ; Was: 1.0
                  (Rendering|Material|SetScalarParameterValueonMaterials _returnvalue_2 "Fuzz Fresnel Blend" 0.5) ; Was: 1.0
                ))))
          (:CastFailed
            (Development|PrintString "InfernoMaterialCastFailed"))))
      (:then_2
        (Niagara|SetTextureObject _asniagara_particle_system_component "BackHairRootPos" "/BRCosmetics/Characters/Player/Male/Medium/Bodies/M_MED_Inferno/FX/Mat/T_BackHairPos.T_BackHairPos")
        (Niagara|SetNiagaraVariable(StaticMesh) _asniagara_particle_system_component "SM_Flame" "/Game/Tools/SplineMeshGenerator/Saves/Meshes/SM_Character_Inferno_BackHair.SM_Character_Inferno_BackHair")
        (Niagara|SetNiagaraVariable(StaticMesh) _asniagara_particle_system_component "SM_BackHair" "/Game/Tools/SplineMeshGenerator/Saves/Meshes/SM_Character_Inferno_BackHair2.SM_Character_Inferno_BackHair2")))
    (else
      (Utilities|FlowControl|Sequence
        (:then_0
          (bind _index_2 (Class|FortCosmeticFunctionLibrary|GetVariantChannelActiveIndexForSlot "(TagName=\"LoadoutSlot.Character\")" (CosmeticLoadout|GetCosmeticLoadout) "(TagName=\"Cosmetics.Variant.Channel.Emissive\")" (Utilities|Array|MakeArray "(TagName=\"Cosmetics.Variant.Property.Emissive1\")" "(TagName=\"Cosmetics.Variant.Property.Emissive2\")")))
          (if (== _index_2 0)
            (bind (_statname _statvalue) (Cosmetics|WatchCosmeticStat (CustomCharacter|GetAssociatedPlayerPawn) "(TagName=\"StatManager.AthenaKills\")")
              (:OnStatChanged
                (if (> _statvalue 0)
                  (Effects|Components|ParticleSystem|SetFloatParameter (CallFunction|GetIdleVFXComponent) "BurstTime" (Utilities|Time|GetGameTimeinSeconds)))))))
        (:then_1
          (bind _index_3 (Class|FortCosmeticFunctionLibrary|GetVariantChannelActiveIndexForSlot "(TagName=\"LoadoutSlot.Character\")" (CosmeticLoadout|GetCosmeticLoadout) "(TagName=\"Cosmetics.Variant.Channel.Material\")" (Utilities|Array|MakeArray "(TagName=\"Cosmetics.Variant.Property.Mat1\")" "(TagName=\"Cosmetics.Variant.Property.Mat2\")" "(TagName=\"Cosmetics.Variant.Property.Mat3\")" "(TagName=\"Cosmetics.Variant.Property.Mat4\")" "(TagName=\"Cosmetics.Variant.Property.Mat5\")")))
          (Utilities|FlowControl|Sequence
            (:then_0
              (if (> _index_3 1)
                (CallFunction|SetupSuperLevel (- _index_3 2))
                (else
                  (Niagara|SetNiagaraVariable(Material) _asniagara_particle_system_component_2 "Mat_BackFlame" (Selection|SelectMaterialInstance "/BRCosmetics/Characters/Player/Male/Medium/Bodies/M_MED_Inferno/FX/Mat/MI_Inferno_FlameHair_Back.MI_Inferno_FlameHair_Back" "/BRCosmetics/Characters/Player/Male/Medium/Bodies/M_MED_Inferno/FX/Mat/MI_Inferno_FlameHair_Back_White.MI_Inferno_FlameHair_Back_White" (== _index_3 0)))
                  (Niagara|SetNiagaraVariable(Material) _asniagara_particle_system_component_2 "Mat_BackHair" (Selection|SelectMaterialInstance "/BRCosmetics/Characters/Player/Male/Medium/Bodies/M_MED_Inferno/FX/Mat/MI_Inferno_MainHair_BackHair.MI_Inferno_MainHair_BackHair" "/BRCosmetics/Characters/Player/Male/Medium/Bodies/M_MED_Inferno/FX/Mat/MI_Inferno_MainHair_BackHair_White_NoColourful.MI_Inferno_MainHair_BackHair_White_NoColourful" (== _index_3 0)))
                  (Niagara|SetNiagaraVariable(Material) _asniagara_particle_system_component_2 "Mat_EyeSpark" (Selection|SelectMaterialInstance "/BRCosmetics/Characters/Player/Male/Medium/Bodies/M_MED_Inferno/FX/Mat/M_Inferno_CrossHair_ForUse.M_Inferno_CrossHair_ForUse" "/BRCosmetics/Characters/Player/Male/Medium/Bodies/M_MED_Inferno/FX/Mat/M_Inferno_CrossHair_White.M_Inferno_CrossHair_White" (== _index_3 0)))
                  (Niagara|SetNiagaraVariable(Material) _asniagara_particle_system_component_2 "Mat_FstLayer" (Selection|SelectMaterialInstance "/BRCosmetics/Characters/Player/Male/Medium/Bodies/M_MED_Inferno/FX/Mat/MI_Inferno_MainHair_FstLayer.MI_Inferno_MainHair_FstLayer" "/BRCosmetics/Characters/Player/Male/Medium/Bodies/M_MED_Inferno/FX/Mat/MI_Inferno_MainHair_FstLayer_White.MI_Inferno_MainHair_FstLayer_White" (== _index_3 0)))
                  (Niagara|SetNiagaraVariable(Material) _asniagara_particle_system_component_2 "Mat_HighLayer" (Selection|SelectMaterialInstance "/BRCosmetics/Characters/Player/Male/Medium/Bodies/M_MED_Inferno/FX/Mat/MI_Inferno_MainHair_HighHair.MI_Inferno_MainHair_HighHair" "/BRCosmetics/Characters/Player/Male/Medium/Bodies/M_MED_Inferno/FX/Mat/MI_Inferno_MainHair_HighHair_White.MI_Inferno_MainHair_HighHair_White" (== _index_3 0)))
                  (Niagara|SetNiagaraVariable(Material) _asniagara_particle_system_component_2 "Mat_MainFlame" (Selection|SelectMaterialInstance "/BRCosmetics/Characters/Player/Male/Medium/Bodies/M_MED_Inferno/FX/Mat/M_Inferno_FlameHair_ForUse.M_Inferno_FlameHair_ForUse" "/BRCosmetics/Characters/Player/Male/Medium/Bodies/M_MED_Inferno/FX/Mat/M_Inferno_FlameHair_White.M_Inferno_FlameHair_White" (== _index_3 0)))
                  (Niagara|SetNiagaraVariable(Material) _asniagara_particle_system_component_2 "Mat_SecLayer" (Selection|SelectMaterialInstance "/BRCosmetics/Characters/Player/Male/Medium/Bodies/M_MED_Inferno/FX/Mat/MI_Inferno_MainHair_SecLayer.MI_Inferno_MainHair_SecLayer" "/BRCosmetics/Characters/Player/Male/Medium/Bodies/M_MED_Inferno/FX/Mat/MI_Inferno_MainHair_SecLayer_White.MI_Inferno_MainHair_SecLayer_White" (== _index_3 0)))
                  (Niagara|SetNiagaraVariable(Material) _asniagara_particle_system_component_2 "Mat_TempEmissiveArea" (Selection|SelectMaterialInstance "/BRCosmetics/Characters/Player/Male/Medium/Bodies/M_MED_Inferno/FX/Mat/M_Inferno_TempCenterEmissive_ForUse.M_Inferno_TempCenterEmissive_ForUse" "/BRCosmetics/Characters/Player/Male/Medium/Bodies/M_MED_Inferno/FX/Mat/M_Inferno_TempCenterEmissive_White.M_Inferno_TempCenterEmissive_White" (== _index_3 0)))
                  (Niagara|SetNiagaraVariable(Material) _asniagara_particle_system_component_2 "Mat_TempEmissiveTranslucent" (Selection|SelectMaterialInstance "/BRCosmetics/Characters/Player/Male/Medium/Bodies/M_MED_Inferno/FX/Mat/MI_Inferno_TempCenterEmissive.MI_Inferno_TempCenterEmissive" "/BRCosmetics/Characters/Player/Male/Medium/Bodies/M_MED_Inferno/FX/Mat/MI_Inferno_TempCenterEmissive_White.MI_Inferno_TempCenterEmissive_White" (== _index_3 0)))
                  (Niagara|SetNiagaraVariable(Material) _asniagara_particle_system_component_2 "Mat_TopFlame" (Selection|SelectMaterialInstance "/BRCosmetics/Characters/Player/Male/Medium/Bodies/M_MED_Inferno/FX/Mat/MI_Inferno_FlameHair_Top.MI_Inferno_FlameHair_Top" "/BRCosmetics/Characters/Player/Male/Medium/Bodies/M_MED_Inferno/FX/Mat/MI_Inferno_FlameHair_Top_White.MI_Inferno_FlameHair_Top_White" (== _index_3 0)))
                  (Niagara|SetNiagaraVariable(Material) _asniagara_particle_system_component_2 "Mat_LegMat" (Selection|SelectMaterialInstance "/BRCosmetics/Characters/Player/Male/Medium/Bodies/M_MED_Inferno/FX/Mat/MI_Inferno_FlameHair_Leg.MI_Inferno_FlameHair_Leg" "/BRCosmetics/Characters/Player/Male/Medium/Bodies/M_MED_Inferno/FX/Mat/MI_Inferno_FlameHair_Leg_White.MI_Inferno_FlameHair_Leg_White" (== _index_3 0)))
                  (Niagara|SetNiagaraVariable(Material) _asniagara_particle_system_component_2 "Mat_NewFlame" (Selection|SelectMaterialInstance "/BRCosmetics/Gadgets/Assets/VinderTech_GliderChute/Glider_Inferno/FX/MI_Glider_Inferno_MeshFlame_01.MI_Glider_Inferno_MeshFlame_01" "/BRCosmetics/Characters/Player/Male/Medium/Bodies/M_MED_Inferno/FX/Mat/GliderMaterial/MI_Glider_Inferno_MeshFlame_01_White.MI_Glider_Inferno_MeshFlame_01_White" (== _index_3 0)))
                  (Niagara|SetNiagaraVariable(Bool) _asniagara_particle_system_component_2 "bIsSL" false))))
            (:then_1
              (Niagara|SetNiagaraVariable(Bool) _asniagara_particle_system_component_2 "bIsWhiteVersion" (not (== _index_3 0))))))
        (:then_2
          (Niagara|SetTextureObject _asniagara_particle_system_component_3 "BackHairRootPos" "/BRCosmetics/Characters/Player/Male/Medium/Bodies/M_MED_Inferno/FX/Mat/T_BackHairPos.T_BackHairPos")
          (Niagara|SetNiagaraVariable(StaticMesh) _asniagara_particle_system_component_3 "SM_Flame" "/Game/Tools/SplineMeshGenerator/Saves/Meshes/SM_Character_Inferno_BackHair.SM_Character_Inferno_BackHair")
          (Niagara|SetNiagaraVariable(StaticMesh) _asniagara_particle_system_component_3 "SM_BackHair" "/Game/Tools/SplineMeshGenerator/Saves/Meshes/SM_Character_Inferno_BackHair2.SM_Character_Inferno_BackHair2"))))))

(fn ConstructionScript
  (|CallParentFunction))
