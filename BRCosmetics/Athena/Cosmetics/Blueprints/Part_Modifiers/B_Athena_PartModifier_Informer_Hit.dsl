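; Cosmetic part modifier for the Informer character that plays a glitch hit reaction on hit pawn events.
; After the parent call, it skips in the front end or while the HitReaction timeline is playing.
; Otherwise it spawns a Niagara effect on the hit actor's root, oriented to the impact normal, and sets HitVector through the part mesh's inverse transform, with no location wired in.
; The timeline then drives the mesh materials' reaction radius and Glitch scalars.
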(event EventOnHitPawn
  (params (SelfActor :type Actor%20Object%20Reference) (OtherActor :type Actor%20Object%20Reference) (NormalImpulse :type Vector :default "0, 0, 0") (Hit :type Hit%20Result%20Structure%20%28by%20ref%29))
  (bind _returnvalue (CustomCharacter|GetSkeletalMeshForPartOnPlayerPawn))
  (bind (_bblockinghit _binitialoverlap _time _distance _location _impactpoint _normal _impactnormal _physmat _hitactor) (Collision|BreakHitResult Hit))
  (|CallParentFunction SelfActor OtherActor NormalImpulse Hit)
  (if (Components|Timeline|IsPlaying (Variables|B_Athena_PartModifier_Informer_Hit|GetHitReaction))
    _
    (elif (CustomCosmeticModifier|IsinFrontEnd)
      _
      (else
        (Variables|Default|SetLocation _location)
        (Variables|Default|SetImpact%20Rotation (Math|Rotator|MakeRotfromX _impactnormal))
        (Variables|Default|SetAttachment%20Component (Variables|/Script/Engine.Actor|GetRootComponent _hitactor))
        (Effects|SpawnFXSystemAttached "/BRCosmetics/Characters/Player/Male/Medium/Bodies/M_MED_Informer/Effects/NS_Informer_Hit2.NS_Informer_Hit2" (Variables|Default|GetAttachment%20Component) "root" (Variables|Default|GetLocation) (Variables|Default|GetImpact%20Rotation) "1.000000,1.000000,1.000000" "KeepWorldPosition" true "AutoRelease")
        (Rendering|Material|SetVectorParameterValueonMaterials _returnvalue "HitVector" (Math|Transform|InverseTransformLocation (Transformation|GetWorldTransform _returnvalue)))
        (bind (_direction _newtrack_0) (Timeline|HitReaction PlayFromStart)
          (:Update
            (Rendering|Material|SetScalarParameterValueonMaterials _returnvalue "Hit Reaction Radius" (* _newtrack_0 (Variables|Default|GetHit%20Reaction%20Radius)))
            (Rendering|Material|SetScalarParameterValueonMaterials _returnvalue "Glitch" _newtrack_0)))))))

(fn ConstructionScript
  (|CallParentFunction))
