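; Character part modifier for the InstantGravel cosmetic that drives speed-reactive materials on the player pawn's part mesh.
; In game, tick accumulates pawn velocity, weighted higher above a horizontal sprint threshold, clamps and slowly decays it, then writes it and current speed to material parameters.
; In frontend, a timeline fakes these values and drives idle VFX opacity instead.
; Idle VFX deactivates in vehicles and reactivates on exit.
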(event EventOnPostResIn
  (bind _returnvalue (CustomCharacter|GetSkeletalMeshForPartOnPlayerPawn))
  ;[1 1 1] Initialize
  (|CallParentFunction)
  (Variables|Default|SetPlayerPawn (CustomCharacter|GetAssociatedPlayerPawn))
  (bind _output_get (Variables|Default|SetIn%20Frontend (Reactive|Frontend|IsCosmeticObjectDisplayedinFrontendContext 0)))
  (if _output_get
    ;[1 1 1] FE preview cycles
    (bind (_direction _alpha) (Timeline|FE_Cycles PlayFromStart)
      (:Update
        (Rendering|Material|SetScalarParameterValueonMaterials _returnvalue "IG_SpeedAccum_CCPM" (* _alpha (Variables|Default|GetMaxVelocityAccum)))
        (Rendering|Material|SetScalarParameterValueonMaterials _returnvalue "IG_Speed_CCPM" (* _alpha 1e+03))
        (Effects|Components|ParticleSystem|SetFloatParameter (CallFunction|GetIdleVFXComponent) "FrontendOpacityMultiplier" _alpha)))
    (else
      (Actor|Tick|SetActorTickEnabled true))))

(event EventTick
  (params (DeltaSeconds :type Float%20%28single-precision%29))
  (bind _returnvalue (CustomCharacter|GetSkeletalMeshForPartOnPlayerPawn))
  ;[1 1 1] Velocity Material and FX Params
  (if (not (Variables|Default|GetIn%20Frontend))
    (bind _output_get (Variables|Default|SetVelocityAccum (+ (Math|Float|Lerp (* (Math|Vector|VectorLength (Transformation|GetVelocity (CustomCharacter|GetAssociatedPlayerPawn))) 0.75) (* (Math|Vector|VectorLength (Transformation|GetVelocity (CustomCharacter|GetAssociatedPlayerPawn))) 2.25) (Math|Conversions|ToFloat(Boolean) (> (Math|Vector|VectorLength (* (Transformation|GetVelocity (CustomCharacter|GetAssociatedPlayerPawn)) "1.000000,1.000000,0.000000")) (Variables|Default|GetSprintBurstTrigger)))) (Variables|Default|GetVelocityAccum))))
    (Utilities|FlowControl|Sequence
      (:then_0
        (Variables|Default|SetVelocityAccum (Math|Interpolation|FInterpTo (Math|Float|Clamp(Float) _output_get 0.0 (Variables|Default|GetMaxVelocityAccum)) 0.0 DeltaSeconds 0.15)))
      (:then_1
        ;[1 1 1] Material update
        (Rendering|Material|SetScalarParameterValueonMaterials _returnvalue "IG_SpeedAccum_CCPM" (Variables|Default|GetVelocityAccum))
        (Rendering|Material|SetScalarParameterValueonMaterials _returnvalue "IG_Speed_CCPM" (Math|Vector|VectorLength (Transformation|GetVelocity (CustomCharacter|GetAssociatedPlayerPawn))))))))

(event EventOnEnterVehicle
  (params (VehicleEntered :type Fort%20Vehicle%20Interface%20Interface%20%28by%20ref%29))
  ;[1 1 1] VehicleCondtitions
  (Components|Activation|Deactivate (CallFunction|GetIdleVFXComponent)))

(event EventOnExitVehicle
  ;[1 1 1] VehicleCondtitions
  (Components|Activation|Activate (CallFunction|GetIdleVFXComponent) false))

(fn ConstructionScript
  (|CallParentFunction))
