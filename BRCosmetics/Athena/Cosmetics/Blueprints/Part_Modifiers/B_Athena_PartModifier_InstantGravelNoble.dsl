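; Cosmetic part modifier for the InstantGravelNoble character part that makes the mesh glow in response to the player pawn's movement speed.
; ReactivityUpdate compares speed to context-specific thresholds, ramping or decaying SprintingGlowAmount and firing a delayed sprint burst with shimmer and pulse material changes.
; In game, Tick feeds pawn velocity; in frontend preview, a timeline fakes speed and toggles a particle parameter.
; Idle VFX deactivates on entering vehicles and reactivates on exit.
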(event EventOnPostResIn
  ;[1 1 1] Initialize
  (|CallParentFunction)
  (Variables|Default|SetPlayerPawn (CustomCharacter|GetAssociatedPlayerPawn))
  (bind _output_get (Variables|Default|SetIn%20Frontend (Reactive|Frontend|IsCosmeticObjectDisplayedinFrontendContext 0)))
  (if _output_get
    ;[1 1 1] FE preview cycles
    (bind (_direction _alpha) (Timeline|Timeline PlayFromStart)
      (:Update
        (Variables|Default|SetCurrentSpeed _alpha)
        (CallFunction|ReactivityUpdate)
        (Effects|Components|ParticleSystem|SetBoolParameter (CallFunction|GetIdleVFXComponent) "FrontEndReactivity" (>= _alpha 0.5))))
    (else
      (Actor|Tick|SetActorTickEnabled true))))

(event EventTick
  (params (DeltaSeconds :type Float%20%28single-precision%29))
  ;[1 0.7412 0.09896] Velocity Material and FX Params
  (if (not (Variables|Default|GetIn%20Frontend))
    (Variables|Default|SetCurrentSpeed (Math|Vector|VectorLength (Transformation|GetVelocity (CustomCharacter|GetAssociatedPlayerPawn))))
    (CallFunction|ReactivityUpdate)))

(event Custom|ReactivityUpdate
  (bind _returnvalue (CustomCharacter|GetSkeletalMeshForPartOnPlayerPawn))
  (bind _returnvalue_2 (Utilities|Time|GetWorldDeltaSeconds))
  (bind _returnvalue_3 (Utilities|Time|GetWorldDeltaSeconds))
  (bind _returnvalue_4 (CustomCharacter|GetSkeletalMeshForPartOnPlayerPawn))
  (bind _returnvalue_5 (* _returnvalue_2 0.25))
  (bind _maxspeedthreshold (Variables|Default|GetMaxSpeedThreshold))
  (bind _currentspeed (Variables|Default|GetCurrentSpeed))
  (bind _returnvalue_6 (< _currentspeed _maxspeedthreshold))
  (bind _sprintburstintensity (Variables|Default|GetSprintBurstIntensity))
  (bind _sprintburstbuildup (Variables|Default|GetSprintBurstBuildUp))
  (bind _glowtransitionalpha (Variables|Default|GetGlowTransitionAlpha))
  ;[1 0.7412 0.09896] Velocity Material and FX Params
  (if (Variables|Default|GetIn%20Frontend)
    ;;[1 1 1] remapping values for the front end context
    (Variables|Default|SetSprintBurstMaxGlowIntensity 4.5)
    (Variables|Default|SetSprintBurstMinGlowIntensity 0.38)
    (Variables|Default|SetBaseSpeedThreshold 0.35) ; Speed in which the glow starts showing up / Jog Speed
    (Variables|Default|SetMaxSpeedThreshold 0.85) ; Speed in which the glow is maximum / Sprint Speed
    (Variables|Default|SetMinSpeedThreshold 0.3) ; Speed in which the glow starts decaying
    (Variables|Default|SetSprintBurstDelay 0.1) ; Delay before the sprint burst
    (else
      (Variables|Default|SetSprintBurstMaxGlowIntensity 4.5)
      (Variables|Default|SetSprintBurstMinGlowIntensity 0.38)
      (Variables|Default|SetBaseSpeedThreshold 405.0) ; Speed in which the glow starts showing up / Jog Speed
      (Variables|Default|SetMaxSpeedThreshold 5.3e+02) ; Speed in which the glow is maximum / Sprint Speed
      (Variables|Default|SetMinSpeedThreshold 4e+02) ; Speed in which the glow starts decaying
      (Variables|Default|SetSprintBurstDelay 1.5) ; Delay before the sprint burst
    ))
  (if (>= (Variables|Default|GetCurrentSpeed) (Variables|Default|GetBaseSpeedThreshold))
    ;;[1 1 1] StartsGlowing
    (Utilities|FlowControl|Sequence
      (:then_0
        (if (< (Variables|Default|GetGlowTransitionAlpha) 1.0)
          (bind _output_get (Variables|Default|SetGlowTransitionAlpha (+ _returnvalue_3 (Variables|Default|GetGlowTransitionAlpha))))
          (Rendering|Material|SetScalarParameterValueonMaterials _returnvalue_4 "SprintingGlowAmount" (Math|Float|Clamp(Float) _output_get))))
      (:then_1
        (if (and (>= (Variables|Default|GetCurrentSpeed) (Variables|Default|GetMaxSpeedThreshold)) (< _sprintburstbuildup (Variables|Default|GetSprintBurstDelay)))
          (bind _output_get_2 (Variables|Default|SetSprintBurstBuildUp (+ _sprintburstbuildup _returnvalue_3)))
          (if (>= _output_get_2 (Variables|Default|GetSprintBurstDelay))
            (bind _output_get_3 (Variables|Default|SetSprintBurstIntensity (Variables|Default|GetSprintBurstMaxGlowIntensity)))
            (Rendering|Material|SetScalarParameterValueonMaterials _returnvalue_4 "BurstGlowAmount" _output_get_3)
            (Rendering|Material|SetScalarParameterValueonMaterials _returnvalue_4 "ShimmerIntensity" 6.0)
            (Rendering|Material|SetScalarParameterValueonMaterials _returnvalue_4 "SineWavePulsingMinValue" 0.3))))
      (:then_2
        (if _returnvalue_6
          (Rendering|Material|SetScalarParameterValueonMaterials _returnvalue_4 "ShimmerIntensity" 0.1)
          (Rendering|Material|SetScalarParameterValueonMaterials _returnvalue_4 "SineWavePulsingMinValue" 0.95)))
      (:then_3
        (if (and _returnvalue_6 (> _sprintburstintensity (Variables|Default|GetSprintBurstMinGlowIntensity)))
          (Variables|Default|SetSprintBurstBuildUp 0.7)
          (bind _output_get_4 (Variables|Default|SetSprintBurstIntensity (- _sprintburstintensity (* (* _returnvalue_3 1.45) (Variables|Default|GetSprintBurstMaxGlowIntensity)))))
          (Rendering|Material|SetScalarParameterValueonMaterials _returnvalue_4 "BurstGlowAmount" _output_get_4))))
    (elif (< (Variables|Default|GetCurrentSpeed) (Variables|Default|GetMinSpeedThreshold))
      ;;[1 1 1] Decay Glow
      (Utilities|FlowControl|Sequence
        (:then_0
          (if (> _glowtransitionalpha 0)
            (bind _output_get_5 (Variables|Default|SetGlowTransitionAlpha (- _glowtransitionalpha _returnvalue_5)))
            (Rendering|Material|SetScalarParameterValueonMaterials _returnvalue "SprintingGlowAmount" (Math|Float|Clamp(Float) _output_get_5))))
        (:then_1
          (if (> (Variables|Default|GetSprintBurstIntensity) (Variables|Default|GetSprintBurstMinGlowIntensity))
            (Variables|Default|SetSprintBurstBuildUp 0.0)
            (bind _output_get_6 (Variables|Default|SetSprintBurstIntensity (- (Variables|Default|GetSprintBurstIntensity) (* _returnvalue_5 (Variables|Default|GetSprintBurstMaxGlowIntensity)))))
            (Rendering|Material|SetScalarParameterValueonMaterials _returnvalue "BurstGlowAmount" _output_get_6)))))))

(event EventOnEnterVehicle
  (params (VehicleEntered :type Fort%20Vehicle%20Interface%20Interface%20%28by%20ref%29))
  ;[1 1 1] VehicleCondtitions
  (Components|Activation|Deactivate (CallFunction|GetIdleVFXComponent)))

(event EventOnExitVehicle
  ;[1 1 1] VehicleCondtitions
  (Components|Activation|Activate (CallFunction|GetIdleVFXComponent) false))

(fn ConstructionScript
  (|CallParentFunction))
