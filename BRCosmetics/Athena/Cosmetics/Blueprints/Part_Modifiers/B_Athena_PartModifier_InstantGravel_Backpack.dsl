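; Part modifier for the Instant Gravel backpack cosmetic that controls its idle VFX.
; After res-in, it casts the part's anim instance to the pack's AnimBP and binds to its hide and unhide dispatchers, which deactivate and reactivate the effect.
; Each pawn jump increments a counter passed to the idle Niagara component as "JumpCounter", and also prints a leftover-looking debug string.
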(event EventOnPostResIn
  (|CallParentFunction)
  (bind _asskeletal_mesh_component (Utilities|Casting|CastToSkeletalMeshComponent (CustomCharacter|GetSkeletalMeshForPartOnPlayerPawn)))
  (bind _asm_med_instant_gravel_pack_anim_bp (Utilities|Casting|CastToM_MED_InstantGravel_Pack_AnimBP (Components|SkeletalMesh|GetAnimInstance _asskeletal_mesh_component)))
  (Utilities|FlowControl|Sequence
    (:then_0
      (Default|BindEventtoBackpackHideDispatcher _asm_med_instant_gravel_pack_anim_bp (delegate Custom|CE_Backpack%20Hide)))
    (:then_1
      (Default|BindEventtoBackpackUnhideDispatcher _asm_med_instant_gravel_pack_anim_bp (delegate Custom|CE_BackpackUnhide)))))

(event Custom|CE_Backpack%20Hide
  (Components|Activation|Deactivate (CallFunction|GetIdleVFXComponent)))

(event EventOnPawnJumped
  (|CallParentFunction)
  (Development|PrintString "Hello!!!!!!!!!!!!!!!!!!!!" true true "(R=1.000000,G=0.029601,B=0.000000,A=1.000000)")
  (bind _output_get (Variables|Default|SetJumpCount (+ (Variables|Default|GetJumpCount) 1)))
  (Niagara|SetNiagaraVariable(Int32) (Utilities|Casting|CastToNiagaraComponent (CallFunction|GetIdleVFXComponent)) "JumpCounter" _output_get))

(event Custom|CE_BackpackUnhide
  (Components|Activation|Activate (CallFunction|GetIdleVFXComponent) true))

(fn ConstructionScript
  (|CallParentFunction))
