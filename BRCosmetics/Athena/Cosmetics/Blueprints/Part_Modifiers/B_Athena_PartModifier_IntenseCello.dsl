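; Cosmetic part modifier for the IntenseCello character that tweaks its materials to show more hatch lines and dots in the front end.
; After the parent's OnPostResIn logic runs, it checks whether it is in the front end.
; If so, it gets the associated player pawn's part meshes and sets scalar material parameters: a fixed light vector on the Head and a rim dots offset on the Body.
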(event EventOnPostResIn
  (bind _returnvalue (CustomCharacter|GetAssociatedPlayerPawn))
  ; Show more hatch lines and dots in FE
  (|CallParentFunction)
  (if (CustomCosmeticModifier|IsinFrontEnd)
    (Rendering|Material|SetScalarParameterValueonMaterials (Customization|GetSkeletalMeshforPartType _returnvalue "Head" :!class /Script/FortniteGame.FortPlayerPawn) "UseFixedLightVector" 1.0)
    (Rendering|Material|SetScalarParameterValueonMaterials (Customization|GetSkeletalMeshforPartType _returnvalue "Body" :!class /Script/FortniteGame.FortPlayerPawn) "RimDotsOffset" 0.5)))

(fn ConstructionScript
  (|CallParentFunction))
