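; Part modifier for the IntenseCello backpack cosmetic that adjusts its comic-style shading (hatch lines and rim dots) in the front end.
; On the post res-in event it runs the parent logic, then checks whether it is in the front end.
; If so, it sets fixed-light-vector and rim-dot-offset scalar parameters on the part's skeletal mesh materials to show more hatching and dots.
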(event EventOnPostResIn
  (bind _returnvalue (CustomCharacter|GetSkeletalMeshForPartOnPlayerPawn))
  ; Show more hatch lines and dots in FE
  (|CallParentFunction)
  (if (CustomCosmeticModifier|IsinFrontEnd)
    (Rendering|Material|SetScalarParameterValueonMaterials _returnvalue "UseFixedLightVector" 1.0)
    (Rendering|Material|SetScalarParameterValueonMaterials _returnvalue "RimDotsOffset" 0.5)))

(fn ConstructionScript
  (|CallParentFunction))
