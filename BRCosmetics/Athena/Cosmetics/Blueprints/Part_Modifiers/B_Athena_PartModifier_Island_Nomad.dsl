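; Cosmetic part modifier for the Island Nomad character part that triggers a feather effect when the pawn lands from a jump.
; On the landed event it gets the idle VFX component and sets its "Jumped" particle bool parameter to true.
; After a short delay it sets the parameter back to false, so the parameter acts as a brief pulse on the existing effect.
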(event EventOnPawnLanded
  (bind _returnvalue (CallFunction|GetIdleVFXComponent))
  ;[1 0 0.1407] FEATHER ON JUMP END
  (Effects|Components|ParticleSystem|SetBoolParameter _returnvalue "Jumped" true)
  (Utilities|FlowControl|Delay 0.35)
  (Effects|Components|ParticleSystem|SetBoolParameter _returnvalue "Jumped" false))

(fn ConstructionScript)
