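; Cosmetic part modifier Blueprint for the JadeTowelGloss face accessory, extending its parent's post-res-in setup with a toon frontend LUT replace.
; In frontend or preview display, a timeline with on and off delays fakes an audio-reactive pulse.
; The timeline drives the EnableFrontendFakeAudioLoop parameter on the part's materials and idle VFX.
; Otherwise, when the pawn's part mesh uses a skin weight profile, the idle Niagara system restarts with Change_OffPhysics enabled.
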(event EventOnPostResIn
  (|CallParentFunction)
  (Utilities|ToonFrontendLUTReplace
    (:Out
      (if (or (Reactive|Frontend|IsCosmeticObjectDisplayedinFrontendContext 0) (Reactive|Frontend|IsCosmeticObjectDisplayedasPreview 0))
        (Variables|Default|SetBeginLoop true)
        (bind (_direction _in) (Timeline|Timeline PlayFromStart)
          (:Update
            (CallFunction|OnFakeLoopTimelineUpdate _in))
          (:Finished
            (if (Variables|Default|GetBeginLoop)
              (Variables|Default|SetBeginLoop false)
              (Utilities|FlowControl|Delay (Variables|Default|GetFakeAmplitudeOnLength))
              (Timeline|Timeline ReverseFromEnd)
              (else
                (Utilities|FlowControl|Delay (Variables|Default|GetFakeAmplitudeOffLength))))))
        (elif (Components|SkinnedMesh|IsUsingSkinWeightProfile (Customization|GetSkeletalMeshforPartType (CustomCharacter|GetAssociatedPlayerPawn) (CustomCharacter|GetCharacterPartType) :!class /Script/FortniteGame.FortPlayerPawn))
          (bind _asniagara_particle_system_component (Utilities|Casting|CastToNiagaraComponent (CallFunction|GetIdleVFXComponent)))
          (Niagara|SetNiagaraVariable(Bool) _asniagara_particle_system_component "Change_OffPhysics" true)
          (Niagara|ResetSystem _asniagara_particle_system_component)
          (Components|Activation|SetActive _asniagara_particle_system_component true true))))))

(event Custom|OnFakeLoopTimelineUpdate
  (params (TransitionValue :type Float%20%28single-precision%29))
  (Rendering|Material|SetScalarParameterValueonMaterials (CustomCharacter|GetSkeletalMeshForPartOnPlayerPawn) "EnableFrontendFakeAudioLoop" TransitionValue) ; Enable Fake Audio in Frontend for Materials
  (Effects|Components|ParticleSystem|SetFloatParameter (CallFunction|GetIdleVFXComponent) "EnableFrontendFakeAudioLoop" TransitionValue))

(fn ConstructionScript
  (|CallParentFunction))
