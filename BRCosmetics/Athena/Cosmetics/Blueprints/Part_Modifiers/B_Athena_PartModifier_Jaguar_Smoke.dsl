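; Cosmetic character part modifier for Jaguar that drives left and right exhaust smoke and flame effects on a player pawn.
; After res-in, it snaps six Rhino-named effect components to the Exhaust_L and Exhaust_R sockets of the part's skeletal mesh.
; Jump, landing and skydiving events toggle only the right flame and the jump exhausts, asymmetrically between sides.
; The left flame and both smoke components are only attached, never toggled here.
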(event EventOnPawnJumped
  (|CallParentFunction)
  (Components|Activation|Deactivate (Variables|Default|GetP_Rhino_Flame_R))
  (Components|Activation|Activate (Variables|Default|GetP_Rhino_Exhaust_Jump_Local_R) false))

(event EventOnBeginSkydiving
  (|CallParentFunction)
  (Components|Activation|Deactivate (Variables|Default|GetP_Rhino_Exhaust_Jump_Local_L)))

(event EventOnPawnLanded
  (|CallParentFunction)
  (Components|Activation|Activate (Variables|Default|GetP_Rhino_Flame_R) false)
  (Components|Activation|Deactivate (Variables|Default|GetP_Rhino_Exhaust_Jump_Local_L)))

(event EventOnEndSkydiving
  (|CallParentFunction)
  (Components|Activation|Activate (Variables|Default|GetP_Rhino_Exhaust_Jump_Local_L) false))

(event EventOnPostResIn
  (bind _returnvalue (CustomCharacter|GetCharacterPartType))
  (bind _returnvalue_2 (CustomCharacter|GetAssociatedPlayerPawn))
  (bind _returnvalue_3 (Customization|GetSkeletalMeshforPartType _returnvalue_2 _returnvalue :!class /Script/FortniteGame.FortPlayerPawn))
  (|CallParentFunction)
  (Utilities|FlowControl|Sequence
    (:then_0
      (Transformation|AttachComponentToComponent (Variables|Default|GetP_Rhino_Exhaust_Jump_Local_L) _returnvalue_3 "Exhaust_L" "SnapToTarget" "SnapToTarget" "SnapToTarget" true))
    (:then_1
      (Transformation|AttachComponentToComponent (Variables|Default|GetP_Rhino_Flame_L) _returnvalue_3 "Exhaust_L" "SnapToTarget" "SnapToTarget" "SnapToTarget" true))
    (:then_2
      (Transformation|AttachComponentToComponent (Variables|Default|GetP_Rhino_Flame_R) _returnvalue_3 "Exhaust_R" "SnapToTarget" "SnapToTarget" "SnapToTarget" true))
    (:then_3
      (Transformation|AttachComponentToComponent (Variables|Default|GetP_Rhino_Exhaust_Jump_Local_R) _returnvalue_3 "Exhaust_R" "SnapToTarget" "SnapToTarget" "SnapToTarget" true))
    (:then_4
      (Transformation|AttachComponentToComponent (Variables|Default|GetP_Rhino_smoke_L) _returnvalue_3 "Exhaust_L" "SnapToTarget" "SnapToTarget" "SnapToTarget" true))
    (:then_5
      (Transformation|AttachComponentToComponent (Variables|Default|GetP_Rhino_smoke_R) _returnvalue_3 "Exhaust_R" "SnapToTarget" "SnapToTarget" "SnapToTarget" true))))

(fn ConstructionScript)
