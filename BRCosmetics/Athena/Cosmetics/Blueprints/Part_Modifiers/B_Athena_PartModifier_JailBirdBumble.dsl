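; Athena cosmetic part modifier for JailBirdBumble that toggles a jitter state on its idle particle effect as the pawn jumps and lands.
; Jump and land event overrides call their parent, then set the IsJitter float parameter on the idle VFX component to on when airborne from a jump and off on landing.
; The construction script only calls its parent.
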(event EventOnPawnJumped
  (|CallParentFunction)
  (Effects|Components|ParticleSystem|SetFloatParameter (CallFunction|GetIdleVFXComponent) "IsJitter" 1.0))

(event EventOnPawnLanded
  (|CallParentFunction)
  (Effects|Components|ParticleSystem|SetFloatParameter (CallFunction|GetIdleVFXComponent) "IsJitter" 0.0))

(fn ConstructionScript
  (|CallParentFunction))
