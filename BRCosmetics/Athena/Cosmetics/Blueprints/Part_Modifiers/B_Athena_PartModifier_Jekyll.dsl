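; Character part modifier for the Jekyll cosmetic that controls the part material's ripple effect around the ManAndMonster "ToMonster" emote.
; On emote begin, it enables ripple (DisableRipple 0) only if that montage's quantized position on the part's anim instance is within the intro duration.
; Other montages or later positions disable ripple.
; On emote end, it disables ripple again if that montage is still current.
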(event EventOnBeginEmote
  (params (MontageItemDef :type Fort%20Item%20Definition%20Object%20Reference) (PawnEmoting :type Fort%20Pawn%20Object%20Reference))
  (bind _anim_instance (CallFunction|GetAnimInstance))
  (bind _returnvalue (Animation|GetCurrentMontage PawnEmoting :!class /Script/Engine.Character))
  (bind _returnvalue_2 (CustomCharacter|GetSkeletalMeshForPartOnPlayerPawn))
  (Utilities|IsValid _anim_instance
    (:Is%20Valid
      (if (Utilities|Equal(Object) _returnvalue "/BRCosmetics/Animation/Game/MainPlayer/Emotes/ManAndMonster/Emote_ManAndMonster_ToMonster_CMM_M.Emote_ManAndMonster_ToMonster_CMM_M")
        (if (<= (/ (DisplayHelper|RoundFloat (* (Animation|Montage|MontageGetPosition _anim_instance _returnvalue) (Variables|Default|GetEmoteTimeQuant))) (Variables|Default|GetEmoteTimeQuant)) (Variables|Default|GetEmoteIntroDuration))
          (Rendering|Material|SetScalarParameterValueonMaterials _returnvalue_2 "DisableRipple" 0.0)
          (else
            (Rendering|Material|SetScalarParameterValueonMaterials _returnvalue_2 "DisableRipple" 1.0)))
        (else
          (Rendering|Material|SetScalarParameterValueonMaterials (CustomCharacter|GetSkeletalMeshForPartOnPlayerPawn) "DisableRipple" 1.0))))))

(event EventOnEndEmote
  (params (MontageItemDef :type Fort%20Item%20Definition%20Object%20Reference) (PawnEmoting :type Fort%20Pawn%20Object%20Reference))
  (if (Utilities|Equal(Object) (Animation|GetCurrentMontage PawnEmoting :!class /Script/Engine.Character) "/BRCosmetics/Animation/Game/MainPlayer/Emotes/ManAndMonster/Emote_ManAndMonster_ToMonster_CMM_M.Emote_ManAndMonster_ToMonster_CMM_M")
    (Rendering|Material|SetScalarParameterValueonMaterials (CustomCharacter|GetSkeletalMeshForPartOnPlayerPawn) "DisableRipple" 1.0)))

(fn GetAnimInstance
  (return (Components|SkeletalMesh|GetAnimInstance (Customization|GetSkeletalMeshforPartType (CustomCharacter|GetAssociatedPlayerPawn) (CustomCharacter|GetCharacterPartType) :!class /Script/FortniteGame.FortPlayerPawn))))

(fn ConstructionScript
  (|CallParentFunction))
