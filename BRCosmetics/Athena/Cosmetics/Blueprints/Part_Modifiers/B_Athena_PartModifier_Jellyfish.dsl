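; Cosmetic part modifier for the Jellyfish character that toggles the "Bubbles_big" emitter on its idle VFX component.
; Vehicle enter and exit events, plus character movement updates that check whether the pawn is falling, drive the emitter on or off.
; DoOnce nodes gate each SetEmitterEnable call, with the vehicle events chaining opposite on and off calls behind nested gates.
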(event EventOnEnterVehicle
  (params (VehicleEntered :type Fort%20Vehicle%20Interface%20Interface%20%28by%20ref%29))
  (bind _returnvalue (CallFunction|GetIdleVFXComponent))
  (Utilities|FlowControl|DoOnce false
    (:Completed
      (Effects|Components|ParticleSystem|SetEmitterEnable _returnvalue "Bubbles_big" false)
      (Utilities|FlowControl|DoOnce false
        (:Completed
          (Effects|Components|ParticleSystem|SetEmitterEnable _returnvalue "Bubbles_big" true))))))

(event EventOnCharacterMovementUpdated
  (params (DeltaSeconds :type Float%20%28single-precision%29) (OldLocation :type Vector :default "0, 0, 0") (OldVelocity :type Vector :default "0, 0, 0"))
  (bind _returnvalue (CallFunction|GetIdleVFXComponent))
  (if (AI|Components|NavMovement|IsFalling (Variables|/Script/Engine.Character|GetCharacterMovement (CustomCharacter|GetAssociatedPlayerPawn)))
    (Utilities|FlowControl|DoOnce false
      (:Completed
        (Effects|Components|ParticleSystem|SetEmitterEnable _returnvalue "Bubbles_big" false))))
  (Utilities|FlowControl|DoOnce false
    (:Completed
      (Effects|Components|ParticleSystem|SetEmitterEnable _returnvalue "Bubbles_big" true))))

(event EventOnExitVehicle
  (bind _returnvalue (CallFunction|GetIdleVFXComponent))
  (Utilities|FlowControl|DoOnce false
    (:Completed
      (Effects|Components|ParticleSystem|SetEmitterEnable _returnvalue "Bubbles_big" true)
      (Utilities|FlowControl|DoOnce false
        (:Completed
          (Effects|Components|ParticleSystem|SetEmitterEnable _returnvalue "Bubbles_big" false))))))

(fn ConstructionScript)
