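; Cosmetic part modifier for the Season 23 Jumbotron back bling that shows the player's kill count on its screen.
; After res-in it attaches a screen glow to the backpack mesh, creates single- and double-digit screen materials, and watches the AthenaKills stat.
; Each change clamps the count to two digits, swaps the material, starts a timed transition and triggers the glow.
; A retriggerable delay then writes digit parameters.
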(event EventOnDayPhaseChanged
  (params (CurrentDayPhase :type EFortDayPhase%20Enum :default "Morning") (PreviousDayPhase :type EFortDayPhase%20Enum :default "Morning") (bAtCreation :type Boolean)))

(event EventOnNewPlayerWeaponType
  (params (NewType :type EFortItemType%20Enum :default "WorldItem") (OldType :type EFortItemType%20Enum :default "WorldItem"))
  (|CallParentFunction NewType OldType) ; This node is disabled and will not be called.\nDrag off pins to build functionality.
)

(event EventOnWeaponChanged
  (params (NewWeapon :type Fort%20Weapon%20Object%20Reference) (OldWeapon :type Fort%20Weapon%20Object%20Reference)))

(event EventOnWeaponFired)

(event EventOnNewWeapons
  (params (WeaponsAsItems :type Array%20of%20Fort%20Item%20Object%20References)))

(event EventOnCurieStateAttached
  (params (Tag :type Gameplay%20Tag%20Structure%20%28by%20ref%29)))

(event EventOnCurieStateDetached
  (params (Tag :type Gameplay%20Tag%20Structure%20%28by%20ref%29)))

(event EventOnBeginSkydiving
  (|CallParentFunction) ; This node is disabled and will not be called.\nDrag off pins to build functionality.
)

(event EventOnEndSkydiving
  (|CallParentFunction) ; This node is disabled and will not be called.\nDrag off pins to build functionality.
)

(event EventOnEnterVehicle
  (params (VehicleEntered :type Fort%20Vehicle%20Interface%20Interface%20%28by%20ref%29))
  (|CallParentFunction VehicleEntered) ; This node is disabled and will not be called.\nDrag off pins to build functionality.
)

(event EventOnExitVehicle
  (|CallParentFunction) ; This node is disabled and will not be called.\nDrag off pins to build functionality.
)

(event EventOnTargetingOrFirstPersonCameraChanged)

(event EventOnHitPawn
  (params (SelfActor :type Actor%20Object%20Reference) (OtherActor :type Actor%20Object%20Reference) (NormalImpulse :type Vector :default "0, 0, 0") (Hit :type Hit%20Result%20Structure%20%28by%20ref%29)))

(event EventOnPostResIn
  (bind _killcount (Variables|Default|GetKillCount))
  (bind (_1st_digit _2nd_digit) (CallFunction|SplitIntegertoIndividualDigits _killcount))
  (|CallParentFunction)
  (bind _output_get (Variables|Default|SetBackpackSkelMesh (CustomCharacter|GetSkeletalMeshForPartOnPlayerPawn)))
  (Transformation|AttachComponentToComponent (Variables|Default|GetScreenGlow) _output_get "ScreenLightGlow" "KeepRelative" "KeepRelative" "KeepRelative" true)
  (CallFunction|CreateMIDs)
  (bind (_statname _statvalue) (Cosmetics|WatchCosmeticStat (CustomCharacter|GetAssociatedPlayerPawn) "(TagName=\"StatManager.AthenaKills\")" -1)
    (:OnStatChanged
      (Variables|Default|SetKillCount (Math|Integer|Clamp(Integer) _statvalue 0 99))
      (CallFunction|SetBackpackMID (Variables|Default|GetKillCount))
      (Utilities|FlowControl|Sequence
        (:then_0
          (Rendering|Material|SetScalarParameterValueonMaterials (Variables|Default|GetBackpackSkelMesh) "Timestamp" (Utilities|Time|GetGameTimeinSeconds))
          (Rendering|Material|SetScalarParameterValueonMaterials (Variables|Default|GetBackpackSkelMesh) "TransitionLength" (Variables|Default|GetTransition%20Length)))
        (:then_1
          (Utilities|FlowControl|RetriggerableDelay (Variables|Default|GetDelayBeforeUpdatingCounterInt))
          (if (>= (Variables|Default|GetKillCount) 10)
            (Rendering|Material|SetScalarParameterValueonMaterials (Variables|Default|GetBackpackSkelMesh) "Char0" _1st_digit)
            (Rendering|Material|SetScalarParameterValueonMaterials (Variables|Default|GetBackpackSkelMesh) "Char1" _2nd_digit)
            (else
              (Rendering|Material|SetScalarParameterValueonMaterials (Variables|Default|GetBackpackSkelMesh) "Char0" _2nd_digit)
              (Rendering|Material|SetScalarParameterValueonMaterials (Variables|Default|GetBackpackSkelMesh) "Char1" _2nd_digit))))
        (:then_2
          (Rendering|Material|SetScalarParameterValueonMaterials (Variables|Default|GetScreenGlow) "HT_CrunchVerts" 1.0))))))

(fn CreateMIDs
  (bind _returnvalue (Rendering|Material|CreateDynamicMaterialInstance "/BRCosmetics/Accessories/FORT_Backpacks/Backpack_M_MED_Jumbotron_S20/Skins/S23/Materials/M_Jumbotron23_Green_Screen_Inst.M_Jumbotron23_Green_Screen_Inst" "None" "None" :!class /Script/Engine.KismetMaterialLibrary))
  (Variables|Default|SetDoubleDigitMID _returnvalue)
  (bind _returnvalue_2 (Rendering|Material|CreateDynamicMaterialInstance "/BRCosmetics/Accessories/FORT_Backpacks/Backpack_M_MED_Jumbotron_S20/Skins/S23/Materials/M_Jumbotron23_Green_Screen_SingleDigit_Inst.M_Jumbotron23_Green_Screen_SingleDigit_Inst" "None" "None" :!class /Script/Engine.KismetMaterialLibrary))
  (Variables|Default|SetSingleDigitMID _returnvalue_2))

(fn SetBackpackMID
  (params (KillCount :type Integer))
  (if (>= KillCount 10)
    (Rendering|Material|SetMaterial (Variables|Default|GetBackpackSkelMesh) 1 (Variables|Default|GetDoubleDigitMID))
    (else
      (Rendering|Material|SetMaterial (Variables|Default|GetBackpackSkelMesh) 1 (Variables|Default|GetSingleDigitMID))))
  (return))

(fn SplitIntegerToIndividualDigits
  (params (Int :type Integer))
  (return (Math|Conversions|ToFloat(Integer) (/ (- Int (Math|Integer|%25%28Integer%29 Int 10)) 10)) (Math|Conversions|ToFloat(Integer) (Math|Integer|%25%28Integer%29 Int 10))))

(fn ConstructionScript
  (|CallParentFunction))
