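; Cosmetic part modifier for the KevinCouture body that steps the character through staged emissive glow effects.
; Cosmetic state activation selects stages 1–3, while post-res-in resets to stage 0 and, in the front end, plays through all stages on timed delays.
; Each stage sets mesh emissive parameters, with 1–3 also spawning an attached Niagara burst that receives hidden head, hand and foot flags.
; Stage 3 activates the idle VFX, which stage 0 deactivates.
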(event EventOnPostResIn
  (|CallParentFunction)
  (CallFunction|Stage0)
  (if (CustomCosmeticModifier|IsinFrontEnd)
    (CallFunction|Stage0)
    (Utilities|FlowControl|Delay 4.0)
    (CallFunction|Stage1)
    (Utilities|FlowControl|Delay 4.0)
    (CallFunction|Stage2)
    (Utilities|FlowControl|Delay 4.0)
    (CallFunction|Stage3)
    (Utilities|FlowControl|Delay 4.0)))

(event EventOnCosmeticStateActivated
  (params (Stat%20Name :type Name :default "None") (Stat%20Value :type Integer) (STAGE%20Num%20Being%20Activated :type Integer))
  (|CallParentFunction Stat%20Name Stat%20Value STAGE%20Num%20Being%20Activated)
  (switch int STAGE%20Num%20Being%20Activated
    (:1
      (CallFunction|Stage1))
    (:2
      (CallFunction|Stage2))
    (:3
      (CallFunction|Stage3))))

(event Custom|Stage_1
  (bind _returnvalue (CustomCharacter|GetSkeletalMeshForPartOnPlayerPawn))
  ;[1 1 1] Stage 1
  (Utilities|FlowControl|Delay 0.2)
  (CallFunction|Burst)
  (Rendering|Material|SetScalarParameterValueonMaterials _returnvalue "Emissive_BrightnessMax" -0.1)
  (Rendering|Material|SetScalarParameterValueonMaterials _returnvalue "Emissive_BrightnessMin" 0.0)
  (Rendering|Material|SetScalarParameterValueonMaterials _returnvalue "Emissive mult" 0.35))

(event Custom|Stage_2
  (bind _returnvalue (CustomCharacter|GetSkeletalMeshForPartOnPlayerPawn))
  ;[1 1 1] Stage 2
  (Utilities|FlowControl|Delay 0.2)
  (CallFunction|Burst)
  (Rendering|Material|SetScalarParameterValueonMaterials _returnvalue "Emissive_BrightnessMax" 4.0)
  (Rendering|Material|SetScalarParameterValueonMaterials _returnvalue "Emissive_BrightnessMin" -0.1)
  (Rendering|Material|SetScalarParameterValueonMaterials _returnvalue "Emissive mult" 0.5))

(event Custom|Stage_3
  (bind _returnvalue (CustomCharacter|GetSkeletalMeshForPartOnPlayerPawn))
  ;[1 1 1] Stage 3
  (Utilities|FlowControl|Delay 0.2)
  (CallFunction|Burst)
  (Components|Activation|Activate (CallFunction|GetIdleVFXComponent) false)
  (Rendering|Material|SetScalarParameterValueonMaterials _returnvalue "Emissive_BrightnessMax" 8.0)
  (Rendering|Material|SetScalarParameterValueonMaterials _returnvalue "Emissive_BrightnessMin" 0.1)
  (Rendering|Material|SetScalarParameterValueonMaterials _returnvalue "Emissive mult" 0.8))

(event Custom|Stage_0
  (bind _returnvalue (CustomCharacter|GetSkeletalMeshForPartOnPlayerPawn))
  (Rendering|Material|SetScalarParameterValueonMaterials _returnvalue "Emissive_BrightnessMax" 0.0)
  (Rendering|Material|SetScalarParameterValueonMaterials _returnvalue "Emissive_BrightnessMin" 0.0)
  (Rendering|Material|SetScalarParameterValueonMaterials _returnvalue "Emissive mult" 0.1)
  (Rendering|Material|SetScalarParameterValueonMaterials _returnvalue "Center Opacity" 0.0)
  (Components|Activation|Deactivate (CallFunction|GetIdleVFXComponent)))

(event Custom|Burst
  (bind _returnvalue (PartModifier|SpawnSystem|SpawnSystemAttachedPartModifier "/BRCosmetics/Effects/Fort_Effects/Effects/Characters/Athena_Parts/KevinCouture/NS_KevinCouture_Body_Burst.NS_KevinCouture_Body_Burst" (CustomCharacter|GetSkeletalMeshForPartOnPlayerPawn))
    (:Success
      (Effects|Components|ParticleSystem|SetBoolParameter _returnvalue "bisHeadHidden" (Variables|HiddenPartCheck|GetbIsHeadHidden))
      (Effects|Components|ParticleSystem|SetBoolParameter _returnvalue "bisHandHidden" (Variables|HiddenPartCheck|GetbIsHandHidden))
      (Effects|Components|ParticleSystem|SetBoolParameter _returnvalue "bisFootHidden" (Variables|HiddenPartCheck|GetbIsFootHidden)))))

(fn ConstructionScript
  (|CallParentFunction))
