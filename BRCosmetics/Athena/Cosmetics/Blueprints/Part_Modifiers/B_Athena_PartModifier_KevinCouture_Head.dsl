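; Cosmetic part modifier for the KevinCouture head that changes its material look through four stages (0 to 3).
; Stage 0 applies on post-init, and cosmetic state activation selects stages 1 to 3.
; Each stage sets emissive scalar parameters on the part's skeletal mesh materials; stages 0 and 3 also toggle hide-element parameters.
; In the front end, post-init also steps through all stages once on timed delays as a preview.
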(event EventOnPostResIn
  (|CallParentFunction)
  (CallFunction|Stage0)
  (if (CustomCosmeticModifier|IsinFrontEnd)
    (CallFunction|Stage0)
    (Utilities|FlowControl|Delay 4.0)
    (CallFunction|Stage1)
    (Utilities|FlowControl|Delay 4.0)
    (CallFunction|Stage2)
    (Utilities|FlowControl|Delay 4.0)
    (CallFunction|Stage3)
    (Utilities|FlowControl|Delay 4.0)))

(event EventOnCosmeticStateActivated
  (params (Stat%20Name :type Name :default "None") (Stat%20Value :type Integer) (STAGE%20Num%20Being%20Activated :type Integer))
  (|CallParentFunction Stat%20Name Stat%20Value STAGE%20Num%20Being%20Activated)
  (switch int STAGE%20Num%20Being%20Activated
    (:1
      (CallFunction|Stage1))
    (:2
      (CallFunction|Stage2))
    (:3
      (CallFunction|Stage3))))

(event Custom|Stage_1
  (bind _returnvalue (CustomCharacter|GetSkeletalMeshForPartOnPlayerPawn))
  ;[1 1 1] Stage 1
  (Utilities|FlowControl|Delay 0.2)
  (Rendering|Material|SetScalarParameterValueonMaterials _returnvalue "Emissive_BrightnessMax" -0.1)
  (Rendering|Material|SetScalarParameterValueonMaterials _returnvalue "Emissive_BrightnessMin" 0.0)
  (Rendering|Material|SetScalarParameterValueonMaterials _returnvalue "Emissive mult" 0.35))

(event Custom|Stage_2
  (bind _returnvalue (CustomCharacter|GetSkeletalMeshForPartOnPlayerPawn))
  ;[1 1 1] Stage 2
  (Utilities|FlowControl|Delay 0.2)
  (Rendering|Material|SetScalarParameterValueonMaterials _returnvalue "Emissive_BrightnessMax" 4.0)
  (Rendering|Material|SetScalarParameterValueonMaterials _returnvalue "Emissive_BrightnessMin" -0.1)
  (Rendering|Material|SetScalarParameterValueonMaterials _returnvalue "Emissive mult" 0.5))

(event Custom|Stage_3
  (bind _returnvalue (CustomCharacter|GetSkeletalMeshForPartOnPlayerPawn))
  ;[1 1 1] Stage 3
  (Utilities|FlowControl|Delay 0.2)
  (Rendering|Material|SetScalarParameterValueonMaterials _returnvalue "Emissive_BrightnessMax" 8.0)
  (Rendering|Material|SetScalarParameterValueonMaterials _returnvalue "Emissive_BrightnessMin" 0.1)
  (Rendering|Material|SetScalarParameterValueonMaterials _returnvalue "Emissive mult" 1.0)
  (Rendering|Material|SetScalarParameterValueonMaterials _returnvalue "Hide Element 01" 0.0)
  (Rendering|Material|SetScalarParameterValueonMaterials _returnvalue "Hide Element 02" 0.0)
  (Rendering|Material|SetScalarParameterValueonMaterials _returnvalue "Hide Element 03" 0.0)
  (Rendering|Material|SetScalarParameterValueonMaterials _returnvalue "Hide Element 04" 0.0)
  (Rendering|Material|SetScalarParameterValueonMaterials _returnvalue "Hide Element 05" 0.0)
  (Rendering|Material|SetScalarParameterValueonMaterials _returnvalue "Hide Element 06" 0.0)
  (Rendering|Material|SetScalarParameterValueonMaterials _returnvalue "Hide Element 07" 0.0)
  (Rendering|Material|SetScalarParameterValueonMaterials _returnvalue "Hide Element 08" 1.0))

(event Custom|Stage_0
  (bind _returnvalue (CustomCharacter|GetSkeletalMeshForPartOnPlayerPawn))
  (Rendering|Material|SetScalarParameterValueonMaterials _returnvalue "Emissive_BrightnessMax" 0.0)
  (Rendering|Material|SetScalarParameterValueonMaterials _returnvalue "Emissive_BrightnessMin" 0.0)
  (Rendering|Material|SetScalarParameterValueonMaterials _returnvalue "Emissive mult" 0.1)
  (Rendering|Material|SetScalarParameterValueonMaterials _returnvalue "Hide Element 01" 1.0)
  (Rendering|Material|SetScalarParameterValueonMaterials _returnvalue "Hide Element 02" 1.0)
  (Rendering|Material|SetScalarParameterValueonMaterials _returnvalue "Hide Element 03" 1.0)
  (Rendering|Material|SetScalarParameterValueonMaterials _returnvalue "Hide Element 04" 1.0)
  (Rendering|Material|SetScalarParameterValueonMaterials _returnvalue "Hide Element 05" 1.0)
  (Rendering|Material|SetScalarParameterValueonMaterials _returnvalue "Hide Element 06" 1.0)
  (Rendering|Material|SetScalarParameterValueonMaterials _returnvalue "Hide Element 07" 1.0)
  (Rendering|Material|SetScalarParameterValueonMaterials _returnvalue "Hide Element 08" 0.0))

(fn ConstructionScript
  (|CallParentFunction))
