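; Cosmetic part modifier for the KneelReedy character that drives flipbook eyebrows on an overlay material on the player pawn's head mesh.
; On post res-in, after calling the parent, it gets or creates a dynamic overlay material for the head and applies it.
; Every tick it copies left and right brow index curves from the head's anim instance into matching scalar parameters, so animation selects the brow frames.
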(event EventOnPostResIn
  (bind _returnvalue (CustomCharacter|GetAssociatedPlayerPawn))
  (bind _returnvalue_2 (Customization|GetSkeletalMeshforPartType _returnvalue "Head" :!class /Script/FortniteGame.FortPlayerPawn))
  (|CallParentFunction)
  (bind _returnvalue_3 (Class|FortCosmeticMaterialBlueprintLibrary|GetorCreateDynamicOverlayMaterial _returnvalue_2))
  (Rendering|Material|SetOverlayMaterial _returnvalue_2 _returnvalue_3 false 0 :!class /Script/Engine.MeshComponent))

(event EventTick
  (params (DeltaSeconds :type Float%20%28single-precision%29))
  (bind _returnvalue (CustomCharacter|GetAssociatedPlayerPawn))
  (bind _returnvalue_2 (Customization|GetSkeletalMeshforPartType _returnvalue "Head" :!class /Script/FortniteGame.FortPlayerPawn))
  (bind _returnvalue_3 (Components|SkeletalMesh|GetAnimInstance _returnvalue_2))
  (Class|FortCosmeticMaterialBlueprintLibrary|SetScalarParameterValueonOverlayMaterial _returnvalue_2 "Flipbook_face_L_brow_index" (Animation|Curves|GetCurveValue _returnvalue_3 "Flipbook_face_L_brow_index"))
  (Class|FortCosmeticMaterialBlueprintLibrary|SetScalarParameterValueonOverlayMaterial _returnvalue_2 "Flipbook_face_R_brow_index" (Animation|Curves|GetCurveValue _returnvalue_3 "Flipbook_face_R_brow_index")))

(fn ConstructionScript
  (|CallParentFunction))
