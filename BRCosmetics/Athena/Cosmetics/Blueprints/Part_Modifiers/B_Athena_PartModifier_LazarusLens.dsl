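; Character part modifier for the LazarusLens cosmetic that animates its glowing tattoo when the player starts or stops the LazarusLens emote.
; Emote events play or reverse a timeline feeding UpdatePhase, which sets Tattoo_EmissivePhase on the part's material, creating a dynamic instance if needed.
; After res-in, it zeroes the phase if the emote count is zero; otherwise it fades out from full, delaying first unless the Progressive variant marks a vampire.
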(event EventOnPawnAndPartChanged
  (|CallParentFunction)
  (bind _asfort_player_pawn_athena (Utilities|Casting|CastToFortPlayerPawnAthena (CustomCharacter|GetAssociatedPlayerPawn)))
  (EventDispatchers|BindEventtoOnPawnStartedEmote _asfort_player_pawn_athena (delegate EventOnBeginEmote))
  (EventDispatchers|BindEventtoOnPawnStoppedEmote _asfort_player_pawn_athena (delegate Custom|Event%20On%20Stopped%20Emote)))

(event Custom|Event%20On%20Stopped%20Emote
  (params (MontageItemDef :type Fort%20Item%20Definition%20Object%20Reference) (PawnEmoting :type Fort%20Pawn%20Object%20Reference))
  (if (== MontageItemDef "/BRCosmetics/Athena/Items/Cosmetics/Dances/EID_LazarusLens.EID_LazarusLens")
    (CallFunction|TattooFadeOut false)))

(event EventOnBeginEmote
  (params (MontageItemDef :type Fort%20Item%20Definition%20Object%20Reference) (PawnEmoting :type Fort%20Pawn%20Object%20Reference))
  (if (== MontageItemDef "/BRCosmetics/Athena/Items/Cosmetics/Dances/EID_LazarusLens.EID_LazarusLens")
    (CallFunction|TattooFadeIn)))

(event Custom|TattooFadeIn
  (bind (_direction _phase) (Timeline|Timeline PlayFromStart)
    (:Update
      (CallFunction|UpdatePhase _phase))))

(event Custom|TattooFadeOut
  (params (FromEnd :type Boolean :default true))
  (if FromEnd
    (Timeline|Timeline ReverseFromEnd)
    (else
      (Timeline|Timeline Reverse))))

(event EventOnPostResIn
  (|CallParentFunction)
  (Utilities|FlowControl|Sequence
    (:then_0
      (if (== (Variables|/Script/FortniteGame.FortPawn|GetEmoteCount (CustomCharacter|GetAssociatedPlayerPawn)) 0)
        (CallFunction|UpdatePhase 0.0)
        (else
          (bind _isvampire (CallFunction|IsVampire))
          (if _isvampire
            (CallFunction|TattooFadeOut true)
            (else
              (CallFunction|UpdatePhase 1.0)
              (Utilities|FlowControl|Delay 0.5)
              (CallFunction|TattooFadeOut true))))))))

(event EventOnNewWeapons
  (params (WeaponsAsItems :type Array%20of%20Fort%20Item%20Object%20References)))

(event EventOnCurieStateAttached
  (params (Tag :type Gameplay%20Tag%20Structure%20%28by%20ref%29)))

(event EventOnDayPhaseChanged
  (params (CurrentDayPhase :type EFortDayPhase%20Enum :default "Morning") (PreviousDayPhase :type EFortDayPhase%20Enum :default "Morning") (bAtCreation :type Boolean)))

(event EventOnCurieStateDetached
  (params (Tag :type Gameplay%20Tag%20Structure%20%28by%20ref%29)))

(event EventOnNewPlayerWeaponType
  (params (NewType :type EFortItemType%20Enum :default "WorldItem") (OldType :type EFortItemType%20Enum :default "WorldItem"))
  (|CallParentFunction NewType OldType) ; This node is disabled and will not be called.\nDrag off pins to build functionality.
)

(event EventOnBeginSkydiving
  (|CallParentFunction) ; This node is disabled and will not be called.\nDrag off pins to build functionality.
)

(event EventOnWeaponChanged
  (params (NewWeapon :type Fort%20Weapon%20Object%20Reference) (OldWeapon :type Fort%20Weapon%20Object%20Reference)))

(event EventOnEndSkydiving
  (|CallParentFunction) ; This node is disabled and will not be called.\nDrag off pins to build functionality.
)

(event EventOnWeaponFired)

(event EventOnEnterVehicle
  (params (VehicleEntered :type Fort%20Vehicle%20Interface%20Interface%20%28by%20ref%29))
  (|CallParentFunction VehicleEntered) ; This node is disabled and will not be called.\nDrag off pins to build functionality.
)

(event EventOnExitVehicle
  (|CallParentFunction) ; This node is disabled and will not be called.\nDrag off pins to build functionality.
)

(event EventOnTargetingOrFirstPersonCameraChanged)

(event EventOnHitPawn
  (params (SelfActor :type Actor%20Object%20Reference) (OtherActor :type Actor%20Object%20Reference) (NormalImpulse :type Vector :default "0, 0, 0") (Hit :type Hit%20Result%20Structure%20%28by%20ref%29)))

(fn CheckHead
  (params (Pawn :type Fort%20Player%20Pawn%20Object%20Reference) (Part :type Custom%20Character%20Part%20Object%20Reference) (MeshPart :type Skeletal%20Mesh%20Component%20Object%20Reference) (PartType :type EFortCustomPartType%20Enum :default "Head"))
  (Utilities|IsValid Pawn
    (:Is%20Valid
      (bind _output_get (Variables|Default|SetIsHead (== "/BRCosmetics/Characters/CharacterParts/Male/Medium/Heads/CP_Athena_Head_M_LazarusLens_Wing.CP_Athena_Head_M_LazarusLens_Wing" Part)))
      (Development|PrintString (Utilities|String|Append "@@@  " "is head " (Utilities|String|ToString(Boolean) _output_get))))))

(fn IsVampire
  (bind _asfort_player_pawn_athena (Utilities|Casting|CastToFortPlayerPawnAthena (CustomCharacter|GetAssociatedPlayerPawn)))
  (bind _index (CosmeticLoadout|GetVariantChannelActiveIndexforSlotActor "(TagName=\"LoadoutSlot.Character\")" _asfort_player_pawn_athena "(TagName=\"Cosmetics.Variant.Channel.Progressive\")" (Utilities|Array|MakeArray "(TagName=\"Cosmetics.Variant.Property.Stage1\")" "(TagName=\"Cosmetics.Variant.Property.Stage2\")")))
  (return (== _index 1)))

(fn UpdatePhase
  (params (Value :type Float%20%28single-precision%29))
  (bind _targetmaterialindex (Variables|Default|GetTargetMaterialIndex))
  (bind _returnvalue (CustomCharacter|GetSkeletalMeshForPartOnPlayerPawn))
  (bind _returnvalue_2 (Rendering|Material|GetMaterial _returnvalue _targetmaterialindex))
  (bind _asmaterial_instance_dynamic (Utilities|Casting|CastToMaterialInstanceDynamic _returnvalue_2)
    (:then
      (Rendering|Material|SetScalarParameterValue _asmaterial_instance_dynamic "Tattoo_EmissivePhase" Value :!class /Script/Engine.MaterialInstanceDynamic))
    (:CastFailed
      (bind _returnvalue_3 (Rendering|Material|CreateDynamicMaterialInstance _returnvalue _targetmaterialindex _returnvalue_2 "None" :!class /Script/Engine.PrimitiveComponent))
      (Rendering|Material|SetScalarParameterValue _returnvalue_3 "Tattoo_EmissivePhase" Value :!class /Script/Engine.MaterialInstanceDynamic))))

(fn ConstructionScript
  (|CallParentFunction))
