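; Athena cosmetic part modifier Blueprint for a Leprechaun character part that spawns a coin effect when the player pawn lands.
; On the pawn landed event it calls the parent handler, then spawns the Coins Jump system attached to the part's skeletal mesh on the player pawn.
; The system snaps to the dyn_coins socket with no offset, and the construction script is empty.
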(event EventOnPawnLanded
  (|CallParentFunction)
  (PartModifier|SpawnSystem|SpawnSystemAttachedPartModifier (Variables|Default|GetCoins%20Jump) (CustomCharacter|GetSkeletalMeshForPartOnPlayerPawn) "dyn_coins" "0, 0, 0" "0, 0, 0" "SnapToTarget" true true true "None" true true true))

(fn ConstructionScript)
