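; Cosmetic part modifier Blueprint for Lightning Dragon that tells its idle Niagara effect when the pawn is emoting.
; Emote begin and end events call the parent, cast the idle VFX component to Niagara, and set its "Begin Emote" bool true or false.
; Both events also print debug strings, and the construction script only calls the parent.
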(event EventOnBeginEmote
  (params (MontageItemDef :type Fort%20Item%20Definition%20Object%20Reference) (PawnEmoting :type Fort%20Pawn%20Object%20Reference))
  (|CallParentFunction MontageItemDef PawnEmoting)
  (bind _asniagara_particle_system_component (Utilities|Casting|CastToNiagaraComponent (CallFunction|GetIdleVFXComponent)))
  (Niagara|SetNiagaraVariable(Bool) _asniagara_particle_system_component "Begin Emote" true)
  (Development|PrintString "Begin Emote!!!!!!!!!!!!!!!!!!" true true "(R=1.000000,G=0.018362,B=0.000000,A=1.000000)"))

(event EventOnEndEmote
  (params (MontageItemDef :type Fort%20Item%20Definition%20Object%20Reference) (PawnEmoting :type Fort%20Pawn%20Object%20Reference))
  (|CallParentFunction MontageItemDef PawnEmoting)
  (bind _asniagara_particle_system_component (Utilities|Casting|CastToNiagaraComponent (CallFunction|GetIdleVFXComponent)))
  (Niagara|SetNiagaraVariable(Bool) _asniagara_particle_system_component "Begin Emote" false)
  (Development|PrintString "End Emote!!!!!!!!!!!!!!!!!!"))

(fn ConstructionScript
  (|CallParentFunction))
