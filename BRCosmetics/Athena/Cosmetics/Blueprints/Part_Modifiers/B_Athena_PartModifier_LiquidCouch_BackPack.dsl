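; Cosmetic part modifier Blueprint for the LiquidCouch back bling, driving its VFX, stat reactions and optional pet setup.
; Post-res-in setup attaches idle VFX, applies a front-end LUT fix and binds emote, movement, DBNO, stat and players-left events.
; Landing spawns tagged VFX; vehicle and skydiving events set particle booleans.
; Gameplay-tag checks on parts, weapons and emotes hide body-part FX, while first-person aiming hides idle VFX after a delay.
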(event EventOnPawnAndPartChanged
  ;[1 1 1] Pet Style Setup
  (if (Variables|PartModifierControls|GetIs%20Pet%20Part%20Modifier)
    (Customization|SetPetType (CustomCharacter|GetAssociatedPlayerPawn) (Variables|PartModifierControls|Pet|GetPetClassToSpawn))))

(event EventOnPawnLanded
  ;[1 1 1] LandingEffects
  (|CallParentFunction)
  (if (CallFunction|IsPreventLandingFXPlayingatIncorrectTimesEnabled)
    (CustomCharacter|ShouldPlayLandingFX "(GameplayTags=((TagName=\"Gameplay.Action.Player.InVehicle\"),(TagName=\"Gameplay.Action.Player.AthenaInVehicle\"),(TagName=\"Granted.Athena.EnvItem.HidingProp.GettingInProp\")))" "(GameplayTags=((TagName=\"MovementMode.Parachuting\"),(TagName=\"MovementMode.Skydiving\")))" 3e+01
      (:True
        _))
    (else
      _))
  (bind _returnvalue (CallFunction|GetVFXSpawnParameters "(TagName=\"CosmeticPart.VFX.Landing\")")
    (:Success
      (PartModifier|SpawnSystem|SpawnSystemAttachedPartModifier (Utilities|Struct|BreakCharacterPartModifierVFXSpawnParameters _returnvalue) (.AttachToComponent (Utilities|Struct|BreakCharacterPartModifierVFXSpawnParameters _returnvalue)) (.SocketName (Utilities|Struct|BreakCharacterPartModifierVFXSpawnParameters _returnvalue)) "0, 0, 0" "0, 0, 0" (.LocationType (Utilities|Struct|BreakCharacterPartModifierVFXSpawnParameters _returnvalue)) true (.bAutoActivate (Utilities|Struct|BreakCharacterPartModifierVFXSpawnParameters _returnvalue))))))

(event Custom|On%20Character%20Movement%20Updated
  (params (DeltaSeconds :type Float%20%28single-precision%29) (OldLocation :type Vector) (OldVelocity :type Vector)))

(event Custom|PlayerStartDBNO
  ;[1 1 1] DBNO
  (if (and (Variables|PartModifierControls|WatchCosmeticStat|GetWatch%20Cosmetic%20Stat) (Variables|PartModifierControls|WatchCosmeticStat|GetWatch%20DBNO%20Status))
    (bind _output_get (Variables|PartModifierControls|WatchCosmeticStat|SetWatch%20Cosmetic%20Stat%20Tag "(TagName=\"StatManager.DBNO\")"))
    (CallFunction|AudioStatWatcher "StatValue" 1 _output_get)))

(event EventOnPostResIn
  (bind _returnvalue (CustomCharacter|GetSkeletalMeshForPartOnPlayerPawn))
  (bind _returnvalue_2 (CustomCharacter|GetAssociatedPlayerPawn))
  ;[1 0 0] Front End LUT, Remove after creating auto LUT tool
  (if (CustomCosmeticModifier|IsinFrontEnd)
    (for (_array_element _array_index) (Rendering|Material|GetMaterials _returnvalue)
      (bind _returnvalue_3 (Rendering|Material|CreateDynamicMaterialInstance _returnvalue _array_index _array_element "None" :!class /Script/Engine.PrimitiveComponent))
      (bind _returnvalue_4 (Rendering|Material|GetScalarParameterValue _returnvalue_3 "PBR_Shading" :!class /Script/Engine.MaterialInstanceDynamic))
      (if (not (== _returnvalue_4 1.0))
        (Rendering|Material|SetTextureParameterValue _returnvalue_3 "LUT Volume Texture" "/Game/Packages/Fortress_SharedMaterials/Anime_Master_Material/VT_InverseLUT_FE.VT_InverseLUT_FE")
        (Rendering|Material|SetScalarParameterValue _returnvalue_3 "LUT Multiplier" 1.0 :!class /Script/Engine.MaterialInstanceDynamic)
        (Rendering|Material|SetScalarParameterValue _returnvalue_3 "Mobile LUT Brightness" 1.0 :!class /Script/Engine.MaterialInstanceDynamic)
        (Rendering|Material|SetScalarParameterValue _returnvalue_3 "use 33 Resolution" 0.0 :!class /Script/Engine.MaterialInstanceDynamic)
        (Rendering|Material|SetScalarParameterValue _returnvalue_3 "use Color Adjustment Instead of LUT" 0.0 :!class /Script/Engine.MaterialInstanceDynamic)
        ;[1 1 1] Attaches Idle Effect to associated part type's Skeletal Mesh
        (bind (_outputidlefx _socketname) (CallFunction|SetupIdleVFXComponent))
        (Utilities|IsValid _outputidlefx
          (:Is%20Valid
            (Transformation|AttachComponentToComponent _outputidlefx (CustomCharacter|GetSkeletalMeshForPartOnPlayerPawn) _socketname "KeepRelative" "KeepRelative" "KeepRelative" true)
            (Development|PrintString (Utilities|String|ToString(Name) _socketname) false false "(R=1.000000,G=0.000000,B=0.000000,A=1.000000)")
            ;[1 1 1] Set bisFrontEndPreview for Niagara Systems
            (bind _outputidlefx_2 (CallFunction|SetupIdleVFXComponent))
            (Effects|Components|ParticleSystem|SetBoolParameter _outputidlefx_2 "bisFrontEndPreview" (CustomCosmeticModifier|IsinFrontEnd))
            (bind _outputidlefx_3 (CallFunction|SetupIdleVFXComponent))
            (CallFunction|MatchParentOwnerNoSee _outputidlefx_3 (CustomCharacter|GetSkeletalMeshForPartOnPlayerPawn)))
          (:Is%20Not%20Valid
            _))
        (Utilities|FlowControl|Sequence
          (:then_0
            ;[1 1 1] Emote Reactivity
            (bind _asfort_player_pawn_athena (Utilities|Casting|CastToFortPlayerPawnAthena (CustomCharacter|GetAssociatedPlayerPawn)))
            (EventDispatchers|BindEventtoOnPawnStartedEmote _asfort_player_pawn_athena (EventDispatchers|CreateEvent :!event Custom|On%20Begin%20Emote))
            (EventDispatchers|BindEventtoOnPawnStoppedEmote _asfort_player_pawn_athena (EventDispatchers|CreateEvent :!event Custom|On%20End%20Emote)))
          (:then_1
            ;[1 1 1] On Character Movement Updated Setup
            (if (Variables|PartModifierControls|GetSprint%20Reactive)
              (bind _asfort_player_pawn_athena_2 (Utilities|Casting|CastToFortPlayerPawnAthena (CustomCharacter|GetAssociatedPlayerPawn))
                (:then
                  (Character|BindEventtoOnCharacterMovementUpdated _asfort_player_pawn_athena_2 (delegate Custom|On%20Character%20Movement%20Updated)))
                (:CastFailed
                  (CallFunction|DEBUGPrintString "Cast to Player Pawn Failed - ")))))
          (:then_2
            ;[1 1 1] DBNO
            (if (not (CustomCosmeticModifier|IsinFrontEnd))
              (Utilities|IsValid _returnvalue_2
                (:Is%20Valid
                  (EventDispatchers|AssignOnPlayerStartDBNO _returnvalue_2 (delegate Custom|PlayerStartDBNO))))))
          (:then_3
            ;[1 1 1] Watch Cosmetic Stat
            (if (Variables|PartModifierControls|WatchCosmeticStat|GetWatch%20Cosmetic%20Stat)
              (bind (_statname _statvalue) (Cosmetics|WatchCosmeticStat (CustomCharacter|GetAssociatedPlayerPawn) (Variables|PartModifierControls|WatchCosmeticStat|GetWatch%20Cosmetic%20Stat%20Tag) -1 (Variables|PartModifierControls|WatchCosmeticStat|GetWatch%20Initial%20Stat%20Value) -1.0 (Class|FortCosmeticFunctionLibrary|GetCosmeticStatDelay (Variables|PartModifierControls|WatchCosmeticStat|GetWatch%20Cosmetic%20Stat%20Tag)) (Variables|Default|GetConsume%20Waiting%20Notifications))
                (:OnStatChanged
                  (CallFunction|IntermediateCosmeticStatChanged _statname _statvalue)
                  (CallFunction|AudioStatWatcher _statname _statvalue (Variables|PartModifierControls|WatchCosmeticStat|GetWatch%20Cosmetic%20Stat%20Tag))))))
          (:then_4
            ;[1 1 1] Watch Players Left Alive
            (if (Variables|PartModifierControls|WatchPlayersLeftAlive|GetPlayers%20Left%20Stat%20Watch)
              (bind _asfort_game_state_athena (Utilities|Casting|CastToFortGameStateAthena (Game|GetGameState)))
              (Utilities|IsA(soft) _asfort_game_state_athena "/Game/Athena/Athena_GameState.Athena_GameState_C"
                (:True
                  (Game|BindEventtoOnPlayersLeftChanged _asfort_game_state_athena (EventDispatchers|CreateEvent :!event Custom|Track%20Players%20Left))
                  (if (Math|Byte|Byte>=Byte (BattleRoyaleGamePhaseLogic|GetGamePhase (BattleRoyaleGamePhaseLogic|GetBattleRoyaleGamePhaseLogicComponent)) 3)
                    (CallFunction|OnPlayersLeftChanged (Utilities|Struct|BreakPlayersLeft (Class|FortGameStateAthena|GetPlayersLeft _asfort_game_state_athena :!class /Script/FortniteGame.FortGameStateAthena)))
                    (if (Variables|PartModifierControls|WatchPlayersLeftAlive|GetStage%20Based%20Player%20Left%20Changes)
                      (for (_array_element_2 _array_index_2) (Variables|PartModifierControls|WatchPlayersLeftAlive|GetPlayer%20Left%20Thresholds)
                        (bind _passed_requirements (CallFunction|IsStageStatRequirementsMet _array_index_2 _array_element_2 (Utilities|Struct|BreakPlayersLeft (Class|FortGameStateAthena|GetPlayersLeft _asfort_game_state_athena :!class /Script/FortniteGame.FortGameStateAthena)) (Variables|PartModifierControls|WatchCosmeticStat|GetCurrent%20Active%20Stage%20Number)))
                        (if _passed_requirements
                          (CallFunction|OnCosmeticStateActivated "Players Left Alive" (Utilities|Struct|BreakPlayersLeft (Class|FortGameStateAthena|GetPlayersLeft _asfort_game_state_athena :!class /Script/FortniteGame.FortGameStateAthena)) _array_index_2)
                          (Variables|PartModifierControls|WatchCosmeticStat|SetCurrent%20Active%20Stage%20Number _array_index_2)))))))))
          (:then_5
            ;[0.3823 1 0.7256] Turn off VFX when targeting (ADS) with a weapon that supports first person view.
            (bind _asfort_player_pawn_athena_3 (Utilities|Casting|CastToFortPlayerPawnAthena (CustomCharacter|GetAssociatedPlayerPawn)))
            (Utilities|IsValid (Pawn|GetLocalViewingPlayerController _asfort_player_pawn_athena_3)
              (:Is%20Valid
                (FortPawn|AssignOnWeaponEquippedDelegate _asfort_player_pawn_athena_3 (delegate Custom|OnWeaponEquippedDelegate_Event))))))))))

(event EventOnEnterVehicle
  (params (VehicleEntered :type Fort%20Vehicle%20Interface%20Interface%20%28by%20ref%29))
  ;[1 1 1] Enter / Exit Vehicle
  (Effects|Components|ParticleSystem|SetBoolParameter (CallFunction|GetIdleVFXComponent) "bisInVehicle" true))

(event EventOnBeginSkydiving
  ;[1 1 1] Begin / End Skydiving
  (Utilities|FlowControl|DelayUntilNextTick)
  (Effects|Components|ParticleSystem|SetBoolParameter (CallFunction|GetIdleVFXComponent) "IsSkydiving" true))

(event EventOnEndSkydiving
  ;[1 1 1] Begin / End Skydiving
  (Effects|Components|ParticleSystem|SetBoolParameter (CallFunction|GetVFXComponent "(TagName=\"CosmeticPart.VFX.Idle\")") "IsSkydiving" false))

(event EventOnExitVehicle
  ;[1 1 1] Enter / Exit Vehicle
  (Effects|Components|ParticleSystem|SetBoolParameter (CallFunction|GetIdleVFXComponent) "bisInVehicle" false))

(event Custom|Intermediate%20Cosmetic%20Stat%20Changed
  (params (Stat%20Name :type Name) (Stat%20Value :type Integer))
  ;[1 1 1] Watch Cosmetic Stat
  (if (Variables|PartModifierControls|WatchCosmeticStat|GetStage%20Based%20Stat%20Changes)
    (CallFunction|OnStatChanged Stat%20Name Stat%20Value)
    (for (_array_element _array_index) (Variables|PartModifierControls|WatchCosmeticStat|GetStat%20Thresholds)
      (bind _passed_requirements (CallFunction|IsStageStatRequirementsMet _array_index _array_element Stat%20Value (Variables|PartModifierControls|WatchCosmeticStat|GetCurrent%20Active%20Stage%20Number)))
      (if _passed_requirements
        (CallFunction|OnCosmeticStateActivated Stat%20Name Stat%20Value _array_index)
        (Variables|PartModifierControls|WatchCosmeticStat|SetCurrent%20Active%20Stage%20Number _array_index)))
    (else
      (CallFunction|OnStatChanged Stat%20Name Stat%20Value))))

(event Custom|On%20Cosmetic%20State%20Activated
  (params (Stat%20Name :type Name) (Stat%20Value :type Integer) (STAGE%20Num%20Being%20Activated :type Integer)))

(event Custom|On%20Stat%20Changed
  (params (State%20Name :type Name) (Stat%20Value :type Integer)))

(event EventOnWeaponChanged
  (params (NewWeapon :type Fort%20Weapon%20Object%20Reference) (OldWeapon :type Fort%20Weapon%20Object%20Reference))
  ;[1 1 1] Hidden Parts Logic
  (Default|HideBodyPartVFX NewWeapon OldWeapon (Variables|HiddenPartCheck|GetAdditional%20FX%20Needing%20Part%20Checks)))

(event Custom|On%20Begin%20Emote
  (params (MontageItemDef :type Fort%20Item%20Definition%20Object%20Reference) (PawnEmoting :type Fort%20Pawn%20Object%20Reference))
  ;[1 1 1] Hidden Parts Logic
  (Networking|CanExecuteCosmeticEvents
    (:True
      (Utilities|FlowControl|Delay 0.2)
      (CallFunction|HiddenPartCheck-Emote PawnEmoting))))

(event Custom|On%20End%20Emote
  (params (MontageItemDef :type Fort%20Item%20Definition%20Object%20Reference) (PawnEmoting :type Fort%20Pawn%20Object%20Reference))
  ;[1 1 1] Hidden Parts Logic
  (Networking|CanExecuteCosmeticEvents
    (:True
      (Utilities|FlowControl|Delay 0.2)
      (CallFunction|HiddenPartCheck-Emote PawnEmoting))))

(event EventOnCosmeticPartAppliedPostResIn
  (params (Part :type Custom%20Character%20Part%20Object%20Reference))
  ;[1 1 1] Hidden Parts Logic
  (Utilities|FlowControl|Sequence
    (:then_0
      (CallFunction|HiddenPartCheck-CharacterPart (Variables|HiddenPartCheck|GetAdditional%20FX%20Needing%20Part%20Checks)))
    (:then_1
      ;;[0.15 0 0 0.5] TEMP until Foot Logic can be untangled
      (bind _outparthidden (CallFunction|ShoeFXCheck false (Variables|HiddenPartCheck|GetAdditional%20FX%20Needing%20Part%20Checks) true))
      (Variables|HiddenPartCheck|SetbIsFootHidden _outparthidden))))

(event Custom|Track%20Players%20Left
  (params (NumPlayersLeft :type Integer))
  ;[1 1 1] Watch Players Left Alive
  (if (Math|Byte|Byte>=Byte (BattleRoyaleGamePhaseLogic|GetGamePhase (BattleRoyaleGamePhaseLogic|GetBattleRoyaleGamePhaseLogicComponent)) 3)
    (CallFunction|OnPlayersLeftChanged NumPlayersLeft)
    (if (Variables|PartModifierControls|WatchPlayersLeftAlive|GetStage%20Based%20Player%20Left%20Changes)
      (for (_array_element _array_index) (Variables|PartModifierControls|WatchPlayersLeftAlive|GetPlayer%20Left%20Thresholds)
        (bind _passed_requirements (CallFunction|IsStageStatRequirementsMet _array_index _array_element NumPlayersLeft (Variables|PartModifierControls|WatchCosmeticStat|GetCurrent%20Active%20Stage%20Number)))
        (if _passed_requirements
          (CallFunction|OnCosmeticStateActivated "Players Left Alive" NumPlayersLeft _array_index)
          (Variables|PartModifierControls|WatchCosmeticStat|SetCurrent%20Active%20Stage%20Number _array_index))))))

(event Custom|On%20Players%20Left%20Changed
  (params (Total%20Players%20Left :type Integer)))

(event EventPreDestruction
  ;[1 1 1] Clear Pet Instance before Teleportation on Destruction
  (if (Variables|PartModifierControls|GetIs%20Pet%20Part%20Modifier)
    (Customization|ClearExistingCosmeticPetInstance (CustomCharacter|GetAssociatedPlayerPawn))))

(event Custom|OnWeaponEquippedDelegate_Event
  (params (NewWeapon :type Fort%20Weapon%20Object%20Reference) (PrevWeapon :type Fort%20Weapon%20Object%20Reference))
  (bind _outputdelegate (EventDispatchers|CreateEvent :!event Custom|OnScopedThirdPersonTargetingChanged))
  (bind _outputdelegate_2 (EventDispatchers|CreateEvent :!event Custom|OnWeaponTargetingChanged))
  ;[0.3823 1 0.7256] Turn off VFX when targeting (ADS) with a weapon that supports first person view.
  (bind _output_get (Variables|Default|SetEquippedWeapon NewWeapon))
  (Utilities|FlowControl|Sequence
    (:then_0
      (Utilities|IsValid PrevWeapon
        (:Is%20Valid
          (EventDispatchers|UnbindEventfromOnTargetingChanged PrevWeapon _outputdelegate_2)
          (bind _asfort_weapon_ranged (Utilities|Casting|CastToFortWeaponRanged PrevWeapon))
          (Targeting|UnbindEventfromOnScopedThirdPersonTargetingChanged _asfort_weapon_ranged _outputdelegate))))
    (:then_1
      (EventDispatchers|BindEventtoOnTargetingChanged _output_get _outputdelegate_2)
      (bind _asfort_weapon_ranged_2 (Utilities|Casting|CastToFortWeaponRanged _output_get))
      (Targeting|BindEventtoOnScopedThirdPersonTargetingChanged _asfort_weapon_ranged_2 _outputdelegate))))

(event Custom|OnScopedThirdPersonTargetingChanged
  (params (bNewIsScopedThirdPersonTargeting :type Boolean))
  ;[0.3823 1 0.7256] Turn off VFX when targeting (ADS) with a weapon that supports first person view.
  (CallFunction|OnWeaponTargetingChanged (FortWeapon|IsTargeting (Variables|Default|GetEquippedWeapon))))

(event Custom|OnWeaponTargetingChanged
  (params (bIsTargeting :type Boolean))
  (bind _returnvalue (CustomCharacter|GetAssociatedPlayerPawn))
  (bind _returnvalue_2 (CustomCharacter|GetAssociatedPlayerPawn))
  ;[0.3823 1 0.7256] Turn off VFX when targeting (ADS) with a weapon that supports first person view.
  (bind _output_get (Variables|Default|SetbLocalIsTargeting bIsTargeting))
  (if _output_get
    (CallFunction|ADSBoolLogic true)
    (bind _usesfirstperson (CallFunction|DoesWeaponUseFirstPerson (Variables|Default|GetEquippedWeapon)))
    (if _usesfirstperson
      (Utilities|FlowControl|Delay (Variables|PartModifierControls|GetADS%20Delay%20Duration))
      ;;[0.7406 0.3587 0.07054 0.5] Don't update FX visiblity if the pawn has FX hidden. Otherwise, this may override FX hiding done by other systems
      ;;[1 1 1] Do nothing if we stopped targeting during delay
      (if (Variables|Default|GetbLocalIsTargeting)
        (Utilities|IsValid _returnvalue_2
          (:Is%20Valid
            (if (not (Class|FortPlayerPawn|AreFXHidden _returnvalue_2))
              (bind _outputidlefx (CallFunction|SetupIdleVFXComponent))
              (Utilities|IsValid _outputidlefx
                (:Is%20Valid
                  (Rendering|SetVisibility _outputidlefx false true)
                  (Variables|Default|SetADS%20Visibility false))
                (:Is%20Not%20Valid
                  _)))
            (Variables|Default|SetADS%20Visibility true))
          (:Is%20Not%20Valid
            _)))
      (else
        _))
    (else
      (CallFunction|ADSBoolLogic false)
      (if (not (Variables|Default|GetADS%20Visibility))
        (Utilities|IsValid _returnvalue
          (:Is%20Valid
            (if (not (Class|FortPlayerPawn|AreFXHidden _returnvalue))
              (bind _outputidlefx_2 (CallFunction|SetupIdleVFXComponent))
              (Utilities|IsValid _outputidlefx_2
                (:Is%20Valid
                  (Rendering|SetVisibility _outputidlefx_2 true true)
                  (Variables|Default|SetADS%20Visibility true))
                (:Is%20Not%20Valid
                  _)))
            (Variables|Default|SetADS%20Visibility false))
          (:Is%20Not%20Valid
            _))))))

(fn ADSBoolLogic
  (params (isTargeting :type Boolean))
  (if (Pawn|IsLocallyControlled (CustomCharacter|GetAssociatedPlayerPawn))
    (Effects|Components|ParticleSystem|SetBoolParameter (CallFunction|GetIdleVFXComponent) "bisADS" (Variables|GetisTargeting))))

(fn AudioStatWatcher
  (params (StatName :type Name) (StatValue :type Integer) (StatTag :type Fort%20Stat%20Manager%20Tag%20Structure%20%28by%20ref%29))
  (bind (_success _audio_components) (CallFunction|PlayLibrarySound StatName StatValue StatTag))
  (if _success
    (CallFunction|UpdateAudioStatValue _audio_components "StatValue" StatValue)))

(fn DEBUGPrintString
  (params (Input%20String :type String))
  (Development|PrintString (Utilities|String|Append Input%20String " - Part Type - " (Utilities|String|EnumtoString (CustomCharacter|GetCharacterPartType)) " - Skeletal Mesh - " (Utilities|GetDisplayName (CustomCharacter|GetSkeletalMeshForPartOnPlayerPawn))) true true "(R=1.000000,G=0.000000,B=0.000000,A=1.000000)" 5.0))

(fn DoesWeaponUseFirstPerson
  (params (Weapon :type Fort%20Weapon%20Object%20Reference))
  (bind _asfort_weapon_ranged (Utilities|Casting|CastToFortWeaponRanged Weapon)
    (:then
      (return (Targeting|UseFirstPersonTargeting _asfort_weapon_ranged)))
    (:CastFailed
      (return false))))

(fn GetAssociatedFortPlayerController
  (bind _returnvalue (CustomCharacter|GetAssociatedPlayerPawn))
  (Utilities|FlowControl|Sequence
    (:then_0
      (bind _aspawn (Utilities|Casting|CastToPawn (Vehicle|GetVehicleActor _returnvalue :!class /Script/FortniteGame.FortPlayerPawn)))
      (bind _asfort_player_controller (Utilities|Casting|CastToFortPlayerController (Pawn|GetController _aspawn)))
      (return _asfort_player_controller))
    (:then_1
      (bind _asfort_player_controller_2 (Utilities|Casting|CastToFortPlayerController (Pawn|GetController _returnvalue))
        (:then
          _)
        (:CastFailed
          _))
      (return _asfort_player_controller_2))))

(fn GetWeaponForHiddenPartCheck
  (bind _returnvalue (BlueprintContext|GetFortInventoryContext))
  (bind _returnvalue_2 (Inventory|GetEquippedWeapon _returnvalue))
  (bind _returnvalue_3 (CustomCharacter|GetAssociatedPlayerPawn))
  (bind _returnvalue_4 (Weapon|GetCurrentWeapon _returnvalue_3))
  (if (CallFunction|IsGetWeaponfromAssociatedPawninHiddenPartCheckEnabled)
    (Utilities|IsValid _returnvalue_3
      (:Is%20Valid
        (Utilities|IsValid _returnvalue_4
          (:Is%20Valid
            (return _returnvalue_4)))))
    (else
      (Utilities|IsValid _returnvalue
        (:Is%20Valid
          (Utilities|IsValid _returnvalue_2
            (:Is%20Valid
              (return _returnvalue_2))))))))

(fn HiddenFeetCheck
  (params (FX%20Systems%20associated%20with%20Feet :type Array%20of%20FXSystem%20Component%20Object%20References) (New%20Weapon :type Fort%20Weapon%20Object%20Reference) (Old%20Weapon :type Fort%20Weapon%20Object%20Reference) (Only%20Set%20Niagara%20Bool%20Variable :type Boolean))
  ;[1 1 1] Even if we have now switched off an FX hidden weapon, shoes should always prevent FX
  (bind _outparthidden (CallFunction|ShoeFXCheck false (Variables|GetFX%20Systems%20associated%20with%20Feet) (Variables|GetOnly%20Set%20Niagara%20Bool%20Variable)))
  (if _outparthidden
    (return true)
    (else
      ;[1 1 1 1 25] Hide FX System Compoenets which are associated with the Feet for when weapons replace feet on Character Model, returns Bool for extended functionality
      (bind _isparthidden (Default|HiddenPartCheck-Weapon (Variables|GetFX%20Systems%20associated%20with%20Feet) (Variables|GetNew%20Weapon) (Variables|GetOld%20Weapon) (Variables|GetOnly%20Set%20Niagara%20Bool%20Variable) "(TagName=\"Animation.CharacterParts.HideFX.Feet\")" "bisFootHidden"))
      (return _isparthidden))))

(fn HiddenPartCheck-CharacterPart
  (params (Additional%20FX%20Systems%20needing%20Part%20Hidden%20Checks :type Array%20of%20FXSystem%20Component%20Object%20References))
  (bind _outputidlefx (CallFunction|SetupIdleVFXComponent))
  (Utilities|IsValid _outputidlefx
    (:Is%20Valid
      (bind _output_get (LocalVariables|SetIdleVFXArray (Utilities|Array|MakeArray _outputidlefx))))
    (:Is%20Not%20Valid
      _))
  (Utilities|Array|AppendArray _output_get Additional%20FX%20Systems%20needing%20Part%20Hidden%20Checks)
  (Utilities|FlowControl|Sequence
    (:then_0
      (bind _returnvalue (Class|CosmeticPortInformationComponent|GetGameplayTagsfromAppliedParts (CustomCharacter|GetAssociatedPlayerPawn)))
      (bind _output_get_2 (LocalVariables|SetPlayerPartTags _returnvalue))
      (if (GameplayTags|HasTag _output_get_2 (LocalVariables|GetHand%20Tag) true)
        (bind _outparthidden (CallFunction|SetPartFXVisibility true false (LocalVariables|GetIdleVFXArray) true (Utilities|Array|Get(acopy) (Variables|HiddenPartCheck|Advanced|GetHidden%20FX%20Parameters))))
        (Variables|HiddenPartCheck|Advanced|SetPartHiddenHand _outparthidden)
        (else
          (if (Variables|HiddenPartCheck|Advanced|GetPartHiddenHand)
            (bind _outparthidden_2 (CallFunction|SetPartFXVisibility false true (LocalVariables|GetIdleVFXArray) true (Utilities|Array|Get(acopy) (Variables|HiddenPartCheck|Advanced|GetHidden%20FX%20Parameters))))
            (Variables|HiddenPartCheck|Advanced|SetPartHiddenHand _outparthidden_2)
            (else
              (Variables|HiddenPartCheck|Advanced|SetPartHiddenHand false)))))
      (if (GameplayTags|HasTag (LocalVariables|GetPlayerPartTags) (LocalVariables|GetHead%20Tag) true)
        (bind _outparthidden_3 (CallFunction|SetPartFXVisibility true false (LocalVariables|GetIdleVFXArray) true (Utilities|Array|Get(acopy) (Variables|HiddenPartCheck|Advanced|GetHidden%20FX%20Parameters) 1)))
        (Variables|HiddenPartCheck|Advanced|SetPartHiddenHead _outparthidden_3)
        (else
          (if (Variables|HiddenPartCheck|Advanced|GetPartHiddenHead)
            (bind _outparthidden_4 (CallFunction|SetPartFXVisibility false true (LocalVariables|GetIdleVFXArray) true (Utilities|Array|Get(acopy) (Variables|HiddenPartCheck|Advanced|GetHidden%20FX%20Parameters) 1)))
            (Variables|HiddenPartCheck|Advanced|SetPartHiddenHead _outparthidden_4)
            (else
              (Variables|HiddenPartCheck|Advanced|SetPartHiddenHead false)))))
      (if (Utilities|Enum|Equal(Enum) (CustomCharacter|GetCharacterPartType) "ExtraPart")
        (CallFunction|DEBUGPrintString "Extra Part Trigggered in Part Hidden Check")
        (elif (GameplayTags|HasTag (LocalVariables|GetPlayerPartTags) (LocalVariables|GetFeet%20Tag) true)
          (bind _outparthidden_5 (CallFunction|SetPartFXVisibility true false (LocalVariables|GetIdleVFXArray) true (Utilities|Array|Get(acopy) (Variables|HiddenPartCheck|Advanced|GetHidden%20FX%20Parameters) 2)))
          (Variables|HiddenPartCheck|Advanced|SetPartHiddenFoot _outparthidden_5)
          (else
            (if (Variables|HiddenPartCheck|Advanced|GetPartHiddenFoot)
              (bind _outparthidden_6 (CallFunction|SetPartFXVisibility false true (LocalVariables|GetIdleVFXArray) true (Utilities|Array|Get(acopy) (Variables|HiddenPartCheck|Advanced|GetHidden%20FX%20Parameters) 2)))
              (Variables|HiddenPartCheck|Advanced|SetPartHiddenFoot _outparthidden_6)
              (else
                (Variables|HiddenPartCheck|Advanced|SetPartHiddenFoot false)))))))
    (:then_1
      (if (not (Variables|HiddenPartCheck|Advanced|GetPartHiddenHand))
        (bind _returnvalue_2 (CallFunction|GetWeaponforHiddenPartCheck))
        (Utilities|IsValid _returnvalue_2
          (:Is%20Valid
            (if (Tags|WeaponDataHasMatchingGameplayTag _returnvalue_2 (LocalVariables|GetHand%20Tag))
              (bind _outparthidden_7 (CallFunction|SetPartFXVisibility true true (LocalVariables|GetIdleVFXArray) true (Utilities|Array|Get(acopy) (Variables|HiddenPartCheck|Advanced|GetHidden%20FX%20Parameters))))
              (Variables|HiddenPartCheck|Advanced|SetWeaponHiddenHand _outparthidden_7)
              (elif (Variables|HiddenPartCheck|Advanced|GetWeaponHiddenHand)
                (bind _outparthidden_8 (CallFunction|SetPartFXVisibility false true (LocalVariables|GetIdleVFXArray) true (Utilities|Array|Get(acopy) (Variables|HiddenPartCheck|Advanced|GetHidden%20FX%20Parameters))))
                (Variables|HiddenPartCheck|Advanced|SetWeaponHiddenHand _outparthidden_8)
                (else
                  (Variables|HiddenPartCheck|Advanced|SetWeaponHiddenHand false)))))))
      (if (not (Variables|HiddenPartCheck|Advanced|GetPartHiddenHead))
        (bind _returnvalue_3 (CallFunction|GetWeaponforHiddenPartCheck))
        (Utilities|IsValid _returnvalue_3
          (:Is%20Valid
            (if (Tags|WeaponDataHasMatchingGameplayTag _returnvalue_3 (LocalVariables|GetHead%20Tag))
              (bind _outparthidden_9 (CallFunction|SetPartFXVisibility true true (LocalVariables|GetIdleVFXArray) true (Utilities|Array|Get(acopy) (Variables|HiddenPartCheck|Advanced|GetHidden%20FX%20Parameters) 1)))
              (Variables|HiddenPartCheck|Advanced|SetWeaponHiddenHead _outparthidden_9)
              (elif (Variables|HiddenPartCheck|Advanced|GetWeaponHiddenHead)
                (bind _outparthidden_10 (CallFunction|SetPartFXVisibility false true (LocalVariables|GetIdleVFXArray) true (Utilities|Array|Get(acopy) (Variables|HiddenPartCheck|Advanced|GetHidden%20FX%20Parameters) 1)))
                (Variables|HiddenPartCheck|Advanced|SetWeaponHiddenHead _outparthidden_10)
                (else
                  (Variables|HiddenPartCheck|Advanced|SetWeaponHiddenHead false)))))))
      (if (Utilities|Enum|Equal(Enum) (CustomCharacter|GetCharacterPartType) "ExtraPart")
        _
        (elif (Variables|HiddenPartCheck|Advanced|GetPartHiddenFoot)
          _
          (else
            (bind _returnvalue_4 (CallFunction|GetWeaponforHiddenPartCheck))
            (Utilities|IsValid _returnvalue_4
              (:Is%20Valid
                (if (Tags|WeaponDataHasMatchingGameplayTag _returnvalue_4 (LocalVariables|GetFeet%20Tag))
                  (bind _outparthidden_11 (CallFunction|SetPartFXVisibility true true (LocalVariables|GetIdleVFXArray) true (Utilities|Array|Get(acopy) (Variables|HiddenPartCheck|Advanced|GetHidden%20FX%20Parameters) 2)))
                  (Variables|HiddenPartCheck|Advanced|SetWeaponHiddenFoot _outparthidden_11)
                  (elif (Variables|HiddenPartCheck|Advanced|GetWeaponHiddenFoot)
                    (bind _outparthidden_12 (CallFunction|SetPartFXVisibility false true (LocalVariables|GetIdleVFXArray) true (Utilities|Array|Get(acopy) (Variables|HiddenPartCheck|Advanced|GetHidden%20FX%20Parameters) 2)))
                    (Variables|HiddenPartCheck|Advanced|SetWeaponHiddenFoot _outparthidden_12)
                    (else
                      (Variables|HiddenPartCheck|Advanced|SetWeaponHiddenFoot false))))))))))
    (:then_2
      (Variables|HiddenPartCheck|SetbIsHandHidden (Math|Boolean|ORBoolean (Variables|HiddenPartCheck|Advanced|GetPartHiddenHand) (Variables|HiddenPartCheck|Advanced|GetWeaponHiddenHand) (Variables|HiddenPartCheck|Advanced|GetEmoteHiddenHand)))
      (Variables|HiddenPartCheck|SetbIsHeadHidden (Math|Boolean|ORBoolean (Variables|HiddenPartCheck|Advanced|GetPartHiddenHead) (Variables|HiddenPartCheck|Advanced|GetWeaponHiddenHead) (Variables|HiddenPartCheck|Advanced|GetEmoteHiddenHead)))
      (Variables|HiddenPartCheck|SetbIsFootHidden (Math|Boolean|ORBoolean (Variables|HiddenPartCheck|Advanced|GetPartHiddenFoot) (Variables|HiddenPartCheck|Advanced|GetWeaponHiddenFoot) (Variables|HiddenPartCheck|Advanced|GetEmoteHiddenFoot))))))

(fn HiddenPartCheck-Emote
  (params (Emoting%20Pawn :type Fort%20Pawn%20Object%20Reference) (Additional%20FX%20Systems%20needing%20Part%20Hidden%20Checks :type Array%20of%20FXSystem%20Component%20Object%20References))
  (bind _outputidlefx (CallFunction|SetupIdleVFXComponent))
  (Utilities|IsValid _outputidlefx
    (:Is%20Valid
      (bind _output_get (LocalVariables|SetIdleVFXArray (Utilities|Array|MakeArray _outputidlefx))))
    (:Is%20Not%20Valid
      _))
  (Utilities|Array|AppendArray _output_get Additional%20FX%20Systems%20needing%20Part%20Hidden%20Checks)
  (Utilities|FlowControl|Sequence
    (:then_0
      (bind _returnvalue (Class|CosmeticPortInformationComponent|GetGameplayTagsfromAppliedParts (CustomCharacter|GetAssociatedPlayerPawn)))
      (bind _output_get_2 (LocalVariables|SetPlayerPartTags _returnvalue))
      (if (GameplayTags|HasTag _output_get_2 (LocalVariables|GetHand%20Tag) true)
        (bind _outparthidden (CallFunction|SetPartFXVisibility true false (LocalVariables|GetIdleVFXArray) true (Utilities|Array|Get(acopy) (Variables|HiddenPartCheck|Advanced|GetHidden%20FX%20Parameters))))
        (Variables|HiddenPartCheck|Advanced|SetPartHiddenHand _outparthidden)
        (else
          (if (Variables|HiddenPartCheck|Advanced|GetPartHiddenHand)
            (bind _outparthidden_2 (CallFunction|SetPartFXVisibility false true (LocalVariables|GetIdleVFXArray) true (Utilities|Array|Get(acopy) (Variables|HiddenPartCheck|Advanced|GetHidden%20FX%20Parameters))))
            (Variables|HiddenPartCheck|Advanced|SetPartHiddenHand _outparthidden_2)
            (else
              (Variables|HiddenPartCheck|Advanced|SetPartHiddenHand false)))))
      (if (GameplayTags|HasTag (LocalVariables|GetPlayerPartTags) (LocalVariables|GetHead%20Tag) true)
        (bind _outparthidden_3 (CallFunction|SetPartFXVisibility true false (LocalVariables|GetIdleVFXArray) true (Utilities|Array|Get(acopy) (Variables|HiddenPartCheck|Advanced|GetHidden%20FX%20Parameters) 1)))
        (Variables|HiddenPartCheck|Advanced|SetPartHiddenHead _outparthidden_3)
        (else
          (if (Variables|HiddenPartCheck|Advanced|GetPartHiddenHead)
            (bind _outparthidden_4 (CallFunction|SetPartFXVisibility false true (LocalVariables|GetIdleVFXArray) true (Utilities|Array|Get(acopy) (Variables|HiddenPartCheck|Advanced|GetHidden%20FX%20Parameters) 1)))
            (Variables|HiddenPartCheck|Advanced|SetPartHiddenHead _outparthidden_4)
            (else
              (Variables|HiddenPartCheck|Advanced|SetPartHiddenHead false)))))
      (if (Utilities|Enum|Equal(Enum) (CustomCharacter|GetCharacterPartType) "ExtraPart")
        (CallFunction|DEBUGPrintString "Extra Part Trigggered in Part Hidden Check")
        (elif (GameplayTags|HasTag (LocalVariables|GetPlayerPartTags) (LocalVariables|GetFeet%20Tag) true)
          (bind _outparthidden_5 (CallFunction|SetPartFXVisibility true false (LocalVariables|GetIdleVFXArray) true (Utilities|Array|Get(acopy) (Variables|HiddenPartCheck|Advanced|GetHidden%20FX%20Parameters) 2)))
          (Variables|HiddenPartCheck|Advanced|SetPartHiddenFoot _outparthidden_5)
          (else
            (if (Variables|HiddenPartCheck|Advanced|GetPartHiddenFoot)
              (bind _outparthidden_6 (CallFunction|SetPartFXVisibility false true (LocalVariables|GetIdleVFXArray) true (Utilities|Array|Get(acopy) (Variables|HiddenPartCheck|Advanced|GetHidden%20FX%20Parameters) 2)))
              (Variables|HiddenPartCheck|Advanced|SetPartHiddenFoot _outparthidden_6)
              (else
                (Variables|HiddenPartCheck|Advanced|SetPartHiddenFoot false)))))))
    (:then_1
      (if (not (Variables|HiddenPartCheck|Advanced|GetPartHiddenHand))
        (bind _returnvalue_2 (CallFunction|GetWeaponforHiddenPartCheck))
        (Utilities|IsValid _returnvalue_2
          (:Is%20Valid
            (if (Tags|WeaponDataHasMatchingGameplayTag _returnvalue_2 (LocalVariables|GetHand%20Tag))
              (bind _outparthidden_7 (CallFunction|SetPartFXVisibility true true (LocalVariables|GetIdleVFXArray) true (Utilities|Array|Get(acopy) (Variables|HiddenPartCheck|Advanced|GetHidden%20FX%20Parameters))))
              (Variables|HiddenPartCheck|Advanced|SetWeaponHiddenHand _outparthidden_7)
              (elif (Variables|HiddenPartCheck|Advanced|GetWeaponHiddenHand)
                (bind _outparthidden_8 (CallFunction|SetPartFXVisibility false true (LocalVariables|GetIdleVFXArray) true (Utilities|Array|Get(acopy) (Variables|HiddenPartCheck|Advanced|GetHidden%20FX%20Parameters))))
                (Variables|HiddenPartCheck|Advanced|SetWeaponHiddenHand _outparthidden_8)
                (else
                  (Variables|HiddenPartCheck|Advanced|SetWeaponHiddenHand false)))))))
      (if (not (Variables|HiddenPartCheck|Advanced|GetPartHiddenHead))
        (bind _returnvalue_3 (CallFunction|GetWeaponforHiddenPartCheck))
        (Utilities|IsValid _returnvalue_3
          (:Is%20Valid
            (if (Tags|WeaponDataHasMatchingGameplayTag _returnvalue_3 (LocalVariables|GetHead%20Tag))
              (bind _outparthidden_9 (CallFunction|SetPartFXVisibility true true (LocalVariables|GetIdleVFXArray) true (Utilities|Array|Get(acopy) (Variables|HiddenPartCheck|Advanced|GetHidden%20FX%20Parameters) 1)))
              (Variables|HiddenPartCheck|Advanced|SetWeaponHiddenHead _outparthidden_9)
              (elif (Variables|HiddenPartCheck|Advanced|GetWeaponHiddenHead)
                (bind _outparthidden_10 (CallFunction|SetPartFXVisibility false true (LocalVariables|GetIdleVFXArray) true (Utilities|Array|Get(acopy) (Variables|HiddenPartCheck|Advanced|GetHidden%20FX%20Parameters) 1)))
                (Variables|HiddenPartCheck|Advanced|SetWeaponHiddenHead _outparthidden_10)
                (else
                  (Variables|HiddenPartCheck|Advanced|SetWeaponHiddenHead false)))))))
      (if (Utilities|Enum|Equal(Enum) (CustomCharacter|GetCharacterPartType) "ExtraPart")
        _
        (elif (Variables|HiddenPartCheck|Advanced|GetPartHiddenFoot)
          _
          (else
            (bind _returnvalue_4 (CallFunction|GetWeaponforHiddenPartCheck))
            (Utilities|IsValid _returnvalue_4
              (:Is%20Valid
                (if (Tags|WeaponDataHasMatchingGameplayTag _returnvalue_4 (LocalVariables|GetFeet%20Tag))
                  (bind _outparthidden_11 (CallFunction|SetPartFXVisibility true true (LocalVariables|GetIdleVFXArray) true (Utilities|Array|Get(acopy) (Variables|HiddenPartCheck|Advanced|GetHidden%20FX%20Parameters) 2)))
                  (Variables|HiddenPartCheck|Advanced|SetWeaponHiddenFoot _outparthidden_11)
                  (elif (Variables|HiddenPartCheck|Advanced|GetWeaponHiddenFoot)
                    (bind _outparthidden_12 (CallFunction|SetPartFXVisibility false true (LocalVariables|GetIdleVFXArray) true (Utilities|Array|Get(acopy) (Variables|HiddenPartCheck|Advanced|GetHidden%20FX%20Parameters) 2)))
                    (Variables|HiddenPartCheck|Advanced|SetWeaponHiddenFoot _outparthidden_12)
                    (else
                      (Variables|HiddenPartCheck|Advanced|SetWeaponHiddenFoot false))))))))))
    (:then_2
      (if (GameplayTags|HasTag (GameplayTags|GetOwnedGameplayTags (Variables|GetEmoting%20Pawn) :!class /Script/GameplayTags.BlueprintGameplayTagLibrary) (LocalVariables|GetHand%20Tag) true)
        (bind _outparthidden_13 (CallFunction|SetPartFXVisibility true false (LocalVariables|GetIdleVFXArray) true (Utilities|Array|Get(acopy) (Variables|HiddenPartCheck|Advanced|GetHidden%20FX%20Parameters))))
        (Variables|HiddenPartCheck|Advanced|SetEmoteHiddenHand _outparthidden_13)
        (else
          (if (Variables|HiddenPartCheck|Advanced|GetEmoteHiddenHand)
            (bind _outparthidden_14 (CallFunction|SetPartFXVisibility false true (LocalVariables|GetIdleVFXArray) true (Utilities|Array|Get(acopy) (Variables|HiddenPartCheck|Advanced|GetHidden%20FX%20Parameters))))
            (Variables|HiddenPartCheck|Advanced|SetEmoteHiddenHand _outparthidden_14)
            (else
              (Variables|HiddenPartCheck|Advanced|SetEmoteHiddenHand false)))))
      (if (GameplayTags|HasTag (GameplayTags|GetOwnedGameplayTags (Variables|GetEmoting%20Pawn) :!class /Script/GameplayTags.BlueprintGameplayTagLibrary) (LocalVariables|GetHead%20Tag) true)
        (bind _outparthidden_15 (CallFunction|SetPartFXVisibility true false (LocalVariables|GetIdleVFXArray) true (Utilities|Array|Get(acopy) (Variables|HiddenPartCheck|Advanced|GetHidden%20FX%20Parameters) 1)))
        (Variables|HiddenPartCheck|Advanced|SetEmoteHiddenHead _outparthidden_15)
        (else
          (if (Variables|HiddenPartCheck|Advanced|GetEmoteHiddenHead)
            (bind _outparthidden_16 (CallFunction|SetPartFXVisibility false true (LocalVariables|GetIdleVFXArray) true (Utilities|Array|Get(acopy) (Variables|HiddenPartCheck|Advanced|GetHidden%20FX%20Parameters) 1)))
            (Variables|HiddenPartCheck|Advanced|SetEmoteHiddenHead _outparthidden_16)
            (else
              (Variables|HiddenPartCheck|Advanced|SetEmoteHiddenHead false)))))
      (if (Utilities|Enum|Equal(Enum) (CustomCharacter|GetCharacterPartType) "ExtraPart")
        (CallFunction|DEBUGPrintString "Extra Part Trigggered in Part Hidden Check")
        (elif (GameplayTags|HasTag (GameplayTags|GetOwnedGameplayTags (Variables|GetEmoting%20Pawn) :!class /Script/GameplayTags.BlueprintGameplayTagLibrary) (LocalVariables|GetFeet%20Tag) true)
          (bind _outparthidden_17 (CallFunction|SetPartFXVisibility true false (LocalVariables|GetIdleVFXArray) true (Utilities|Array|Get(acopy) (Variables|HiddenPartCheck|Advanced|GetHidden%20FX%20Parameters) 2)))
          (Variables|HiddenPartCheck|Advanced|SetEmoteHiddenFoot _outparthidden_17)
          (else
            (if (Variables|HiddenPartCheck|Advanced|GetEmoteHiddenFoot)
              (bind _outparthidden_18 (CallFunction|SetPartFXVisibility false true (LocalVariables|GetIdleVFXArray) true (Utilities|Array|Get(acopy) (Variables|HiddenPartCheck|Advanced|GetHidden%20FX%20Parameters) 2)))
              (Variables|HiddenPartCheck|Advanced|SetEmoteHiddenFoot _outparthidden_18)
              (else
                (Variables|HiddenPartCheck|Advanced|SetEmoteHiddenFoot false)))))))
    (:then_3
      (Variables|HiddenPartCheck|SetbIsHandHidden (Math|Boolean|ORBoolean (Variables|HiddenPartCheck|Advanced|GetPartHiddenHand) (Variables|HiddenPartCheck|Advanced|GetWeaponHiddenHand) (Variables|HiddenPartCheck|Advanced|GetEmoteHiddenHand)))
      (Variables|HiddenPartCheck|SetbIsHeadHidden (Math|Boolean|ORBoolean (Variables|HiddenPartCheck|Advanced|GetPartHiddenHead) (Variables|HiddenPartCheck|Advanced|GetWeaponHiddenHead) (Variables|HiddenPartCheck|Advanced|GetEmoteHiddenHead)))
      (Variables|HiddenPartCheck|SetbIsFootHidden (Math|Boolean|ORBoolean (Variables|HiddenPartCheck|Advanced|GetPartHiddenFoot) (Variables|HiddenPartCheck|Advanced|GetWeaponHiddenFoot) (Variables|HiddenPartCheck|Advanced|GetEmoteHiddenFoot))))))

(fn HiddenPartCheck-Weapon
  (params (FX%20Systems%20associated%20with%20Part :type Array%20of%20FXSystem%20Component%20Object%20References) (New%20Weapon :type Fort%20Weapon%20Object%20Reference) (Old%20Weapon :type Fort%20Weapon%20Object%20Reference) (Only%20Set%20Niagara%20Bool%20Variable :type Boolean) (TagToCheck :type Gameplay%20Tag%20Structure) (FXParameterName :type Name))
  (bind _returnvalue (Tags|WeaponDataHasMatchingGameplayTag New%20Weapon TagToCheck))
  (bind _only_set_niagara_bool_variable (Variables|GetOnly%20Set%20Niagara%20Bool%20Variable))
  (bind _fxparametername (Variables|GetFXParameterName))
  (bind _fx_systems_associated_with_part (Variables|GetFX%20Systems%20associated%20with%20Part))
  ;[1 1 1 1 25] Hide FX System Compoenets which are associated with this part for when weapons replace parts on Character Model, returns Bool for extended functionality
  (Utilities|IsValid New%20Weapon
    (:Is%20Valid
      (Utilities|IsValid Old%20Weapon
        (:Is%20Valid
          (bind _outparthidden (CallFunction|SetPartFXVisibility _returnvalue (Tags|WeaponDataHasMatchingGameplayTag Old%20Weapon TagToCheck) _fx_systems_associated_with_part _only_set_niagara_bool_variable _fxparametername))
          (return _outparthidden))
        (:Is%20Not%20Valid
          (bind _outparthidden_2 (CallFunction|SetPartFXVisibility _returnvalue false _fx_systems_associated_with_part _only_set_niagara_bool_variable _fxparametername))
          (return _outparthidden_2))))
    (:Is%20Not%20Valid
      (return false))))

(fn HideBodyPartVFX
  (params (NewWeapon :type Fort%20Weapon%20Object%20Reference) (OldWeapon :type Fort%20Weapon%20Object%20Reference) (Additional%20FX%20Systems%20needing%20Component%20Check :type Array%20of%20FXSystem%20Component%20Object%20References))
  (bind _outputidlefx (CallFunction|SetupIdleVFXComponent))
  (Utilities|IsValid _outputidlefx
    (:Is%20Valid
      (bind _output_get (LocalVariables|SetIdleVFXArray (Utilities|Array|MakeArray _outputidlefx)))
      (Utilities|Array|AppendArray _output_get Additional%20FX%20Systems%20needing%20Component%20Check)
      (Utilities|FlowControl|Sequence
        (:then_0
          (bind _returnvalue (Class|CosmeticPortInformationComponent|GetGameplayTagsfromAppliedParts (CustomCharacter|GetAssociatedPlayerPawn)))
          (bind _output_get_2 (LocalVariables|SetPlayerPartTags _returnvalue))
          (if (GameplayTags|HasTag _output_get_2 (LocalVariables|GetHIdden%20Hand%20Tag) true)
            (bind _outparthidden (CallFunction|SetPartFXVisibility true false (LocalVariables|GetIdleVFXArray) true (Utilities|Array|Get(acopy) (Variables|HiddenPartCheck|Advanced|GetHidden%20FX%20Parameters))))
            (Variables|HiddenPartCheck|Advanced|SetPartHiddenHand _outparthidden)
            (else
              (if (Variables|HiddenPartCheck|Advanced|GetPartHiddenHand)
                (bind _outparthidden_2 (CallFunction|SetPartFXVisibility false true (LocalVariables|GetIdleVFXArray) true (Utilities|Array|Get(acopy) (Variables|HiddenPartCheck|Advanced|GetHidden%20FX%20Parameters))))
                (Variables|HiddenPartCheck|Advanced|SetPartHiddenHand _outparthidden_2)
                (else
                  (Variables|HiddenPartCheck|Advanced|SetPartHiddenHand false)))))
          (if (GameplayTags|HasTag (LocalVariables|GetPlayerPartTags) (LocalVariables|GetHidden%20Head%20Tag) true)
            (bind _outparthidden_3 (CallFunction|SetPartFXVisibility true false (LocalVariables|GetIdleVFXArray) true (Utilities|Array|Get(acopy) (Variables|HiddenPartCheck|Advanced|GetHidden%20FX%20Parameters) 1)))
            (Variables|HiddenPartCheck|Advanced|SetPartHiddenHead _outparthidden_3)
            (else
              (if (Variables|HiddenPartCheck|Advanced|GetPartHiddenHead)
                (bind _outparthidden_4 (CallFunction|SetPartFXVisibility false true (LocalVariables|GetIdleVFXArray) true (Utilities|Array|Get(acopy) (Variables|HiddenPartCheck|Advanced|GetHidden%20FX%20Parameters) 1)))
                (Variables|HiddenPartCheck|Advanced|SetPartHiddenHead _outparthidden_4)
                (else
                  (Variables|HiddenPartCheck|Advanced|SetPartHiddenHead false)))))
          (if (Utilities|Enum|Equal(Enum) (CustomCharacter|GetCharacterPartType) "ExtraPart")
            (CallFunction|DEBUGPrintString "Extra Part Trigggered in Part Hidden Check")
            (elif (GameplayTags|HasTag (LocalVariables|GetPlayerPartTags) (LocalVariables|GetHidden%20Foot%20Tag) true)
              (bind _outparthidden_5 (CallFunction|SetPartFXVisibility true false (LocalVariables|GetIdleVFXArray) true (Utilities|Array|Get(acopy) (Variables|HiddenPartCheck|Advanced|GetHidden%20FX%20Parameters) 2)))
              (Variables|HiddenPartCheck|Advanced|SetPartHiddenFoot _outparthidden_5)
              (else
                (if (Variables|HiddenPartCheck|Advanced|GetPartHiddenFoot)
                  (bind _outparthidden_6 (CallFunction|SetPartFXVisibility false true (LocalVariables|GetIdleVFXArray) true (Utilities|Array|Get(acopy) (Variables|HiddenPartCheck|Advanced|GetHidden%20FX%20Parameters) 2)))
                  (Variables|HiddenPartCheck|Advanced|SetPartHiddenFoot _outparthidden_6)
                  (else
                    (Variables|HiddenPartCheck|Advanced|SetPartHiddenFoot false)))))))
        (:then_1
          (if (not (Variables|HiddenPartCheck|Advanced|GetPartHiddenHand))
            (bind _isparthidden (Default|HiddenPartCheck-Weapon (LocalVariables|GetIdleVFXArray) (Variables|GetNewWeapon) (Variables|GetOldWeapon) true (LocalVariables|GetHIdden%20Hand%20Tag) (Utilities|Array|Get(acopy) (Variables|HiddenPartCheck|Advanced|GetHidden%20FX%20Parameters))))
            (Variables|HiddenPartCheck|Advanced|SetWeaponHiddenHand _isparthidden))
          (if (not (Variables|HiddenPartCheck|Advanced|GetPartHiddenHead))
            (bind _isparthidden_2 (Default|HiddenPartCheck-Weapon (LocalVariables|GetIdleVFXArray) (Variables|GetNewWeapon) (Variables|GetOldWeapon) true (LocalVariables|GetHidden%20Head%20Tag) (Utilities|Array|Get(acopy) (Variables|HiddenPartCheck|Advanced|GetHidden%20FX%20Parameters) 1)))
            (Variables|HiddenPartCheck|Advanced|SetWeaponHiddenHead _isparthidden_2))
          (if (Utilities|Enum|Equal(Enum) (CustomCharacter|GetCharacterPartType) "ExtraPart")
            (CallFunction|DEBUGPrintString "Extra Part Trigggered in Part Hidden Check")
            (elif (Variables|HiddenPartCheck|Advanced|GetPartHiddenFoot)
              _
              (else
                (bind _isparthidden_3 (Default|HiddenPartCheck-Weapon (LocalVariables|GetIdleVFXArray) (Variables|GetNewWeapon) (Variables|GetOldWeapon) true (LocalVariables|GetHidden%20Foot%20Tag) (Utilities|Array|Get(acopy) (Variables|HiddenPartCheck|Advanced|GetHidden%20FX%20Parameters) 2)))
                (Variables|HiddenPartCheck|Advanced|SetWeaponHiddenFoot _isparthidden_3)))))
        (:then_2
          (Variables|HiddenPartCheck|SetbIsHandHidden (Math|Boolean|ORBoolean (Variables|HiddenPartCheck|Advanced|GetPartHiddenHand) (Variables|HiddenPartCheck|Advanced|GetWeaponHiddenHand) (Variables|HiddenPartCheck|Advanced|GetEmoteHiddenHand)))
          (Variables|HiddenPartCheck|SetbIsHeadHidden (Math|Boolean|ORBoolean (Variables|HiddenPartCheck|Advanced|GetPartHiddenHead) (Variables|HiddenPartCheck|Advanced|GetWeaponHiddenHead) (Variables|HiddenPartCheck|Advanced|GetEmoteHiddenHead)))
          (Variables|HiddenPartCheck|SetbIsFootHidden (Math|Boolean|ORBoolean (Variables|HiddenPartCheck|Advanced|GetPartHiddenFoot) (Variables|HiddenPartCheck|Advanced|GetWeaponHiddenFoot) (Variables|HiddenPartCheck|Advanced|GetEmoteHiddenFoot))))))))

(fn IsStageStatRequirementsMet
  (params (In%20Stage%20Number%20Begin%20Tested :type Integer) (In%20Stage%20Threshold :type Integer) (In%20Stat%20Value :type Integer) (Current%20Active%20Stage%20Number :type Integer))
  (return (Math|Boolean|ANDBoolean (< Current%20Active%20Stage%20Number In%20Stage%20Number%20Begin%20Tested) (> In%20Stage%20Threshold 0) (>= In%20Stat%20Value In%20Stage%20Threshold))))

(fn IsPartModifierHidden
  (return (or (not (Rendering|IsVisible (CustomCharacter|GetSkeletalMeshForPartOnPlayerPawn))) (Class|FortPawn|IsInsideHidingProp (CustomCharacter|GetAssociatedPlayerPawn) true))))

(fn MatchParentOwnerNoSee
  (params (Child :type Primitive%20Component%20Object%20Reference) (Parent :type Primitive%20Component%20Object%20Reference))
  (bind _bownernosee (Variables|/Script/Engine.PrimitiveComponent|GetbOwnerNoSee Parent))
  ;[1 1 1] Attached child VFX needs to match OwnerNoSee flag of parent component
  (if (!= (Variables|/Script/Engine.PrimitiveComponent|GetbOwnerNoSee Child) _bownernosee)
    (Rendering|SetOwnerNoSee Child _bownernosee :!class /Script/Engine.PrimitiveComponent)))

(fn PlayLibrarySound
  (params (StatName :type Name) (StatValue :type Integer) (StatTag :type Fort%20Stat%20Manager%20Tag%20Structure%20%28by%20ref%29))
  (bind _statcosmeticmapping (Variables|SoundLibrary|GetStatCosmeticMapping))
  (bind (_value _returnvalue) (Utilities|Map|Find _statcosmeticmapping StatTag))
  (bind _returnvalue_2 (CustomCharacter|GetAssociatedPlayerPawn))
  (if (and (> StatValue 0) _returnvalue)
    (Utilities|IsValid _returnvalue_2
      (:Is%20Valid
        (bind _returnvalue_3 (SoundLibrary|PlaySoundSimple (WorldSubsystems|GetSoundLibrarySubsystem) _returnvalue_2 (Utilities|Struct|MakeSimpleSoundSettings _value true true true false 1.0 1.0 0.0 "/Game/Sounds/Concurrency_Rules/OneMaximum_Owner_StopOldest.OneMaximum_Owner_StopOldest" 0 "None" (CustomCharacter|GetSkeletalMeshForPartOnPlayerPawn)) :!class /Script/SoundLibrary.SoundLibrarySubsystem))
        (if (Utilities|Array|IsValidIndex (.AudioComponents (Utilities|Struct|BreakSoundLibraryPlayResult _returnvalue_3)))
          (return (.bSuccess (Utilities|Struct|BreakSoundLibraryPlayResult _returnvalue_3)) (.AudioComponents (Utilities|Struct|BreakSoundLibraryPlayResult _returnvalue_3)))
          (else
            (return false 0)))))))

(fn SetPartFXVisibility
  (params (IsHidden :type Boolean) (WasHidden :type Boolean) (FXAssociatedWithPart :type Array%20of%20FXSystem%20Component%20Object%20References) (OnlySetNiagara :type Boolean) (FXParameterName :type Name))
  (if (Variables|GetIsHidden)
    ;[1 1 1] Is currently hidden, we should set "Hidden" to true
    (if (> (Utilities|Array|Length (Variables|GetFXAssociatedWithPart)) 0)
      (for _array_element (Variables|GetFXAssociatedWithPart)
        (if (Variables|GetOnlySetNiagara)
          (Effects|Components|ParticleSystem|SetBoolParameter _array_element (Variables|GetFXParameterName) true)
          (else
            (Development|SetHiddeninGame _array_element true true))))
      (return true)
      (else
        (return true)))
    (elif (Variables|GetWasHidden)
      ;[1 1 1] Was Hidden, but no longer is - We should set "Hidden" to false
      (if (> (Utilities|Array|Length (Variables|GetFXAssociatedWithPart)) 0)
        (for _array_element_2 (Variables|GetFXAssociatedWithPart)
          (if (Variables|GetOnlySetNiagara)
            (Effects|Components|ParticleSystem|SetBoolParameter _array_element_2 (Variables|GetFXParameterName) false)
            (else
              (Development|SetHiddeninGame _array_element_2 false true))))
        (return false)
        (else
          (return false)))
      (else
        (return false)))))

(fn SetupIdleVFXComponent
  (bind _returnvalue (CallFunction|GetIdleVFXComponent))
  (bind _returnvalue_2 (CallFunction|GetVFXComponent "(TagName=\"CosmeticPart.VFX.Idle\")"))
  (Utilities|IsValid _returnvalue
    (:Is%20Valid
      (bind _output_get (Variables|Default|SetIdleFX _returnvalue))
      (return _output_get (CallFunction|GetIdleFXSocketName)))
    (:Is%20Not%20Valid
      (Utilities|IsValid _returnvalue_2
        (:Is%20Valid
          (bind _returnvalue_3 (CallFunction|GetVFXSpawnParameters "(TagName=\"CosmeticPart.VFX.Idle\")")
            (:Success
              (bind _output_get_2 (Variables|Default|SetIdleFX _returnvalue_2))
              (return _output_get_2 (.SocketName (Utilities|Struct|BreakCharacterPartModifierVFXSpawnParameters _returnvalue_3))))))))))

(fn ShoeFXCheck
  (params (CurrentlyHidden :type Boolean) (FXSystemsAssociatedWithFeet :type Array%20of%20FXSystem%20Component%20Object%20References) (OnlySetNiagaraBoolVariable :type Boolean))
  (bind _returnvalue (BlueprintContext|GetFortInventoryContext))
  (bind _returnvalue_2 (Inventory|GetEquippedWeapon _returnvalue))
  (bind _outputidlefx (CallFunction|SetupIdleVFXComponent))
  (Utilities|IsValid _outputidlefx
    (:Is%20Valid
      (bind _output_get (LocalVariables|SetIdleVFXArray (Utilities|Array|MakeArray _outputidlefx))))
    (:Is%20Not%20Valid
      _))
  (Utilities|Array|AppendArray _output_get FXSystemsAssociatedWithFeet)
  (if (Variables|HiddenPartCheck|GetbIsFootHidden)
    (return (Variables|HiddenPartCheck|GetbIsFootHidden))
    (else
      (if (Utilities|IsValid (CosmeticLoadout|GetItemDefinitionfromLoadout (CosmeticLoadout|GetCosmeticLoadout) "(TagName=\"LoadoutSlot.ExtraPart\")") :!class /Script/Engine.KismetSystemLibrary)
        (if (Utilities|Enum|Equal(Enum) (CustomCharacter|GetCharacterPartType) "ExtraPart")
          (CallFunction|DEBUGPrintString "Extra Part Trigggered in Part Hidden Check")
          (else
            (bind _outparthidden (CallFunction|SetPartFXVisibility true false (LocalVariables|GetIdleVFXArray) (Variables|GetOnlySetNiagaraBoolVariable) "bisFootHidden"))
            (bind _output_get_2 (Variables|HiddenPartCheck|Advanced|SetPartHiddenFoot _outparthidden))
            (return _output_get_2)))
        (elif (Utilities|Enum|Equal(Enum) (CustomCharacter|GetCharacterPartType) "ExtraPart")
          (CallFunction|DEBUGPrintString "Extra Part Trigggered in Part Hidden Check")
          (elif (Variables|HiddenPartCheck|Advanced|GetPartHiddenFoot)
            (return (Variables|HiddenPartCheck|Advanced|GetPartHiddenFoot))
            (else
              (Utilities|IsValid _returnvalue
                (:Is%20Valid
                  (Utilities|IsValid _returnvalue_2
                    (:Is%20Valid
                      (if (Tags|WeaponDataHasMatchingGameplayTag _returnvalue_2 (LocalVariables|GetFeet%20Tag))
                        (bind _outparthidden_2 (CallFunction|SetPartFXVisibility true true (LocalVariables|GetIdleVFXArray) true (Utilities|Array|Get(acopy) (Variables|HiddenPartCheck|Advanced|GetHidden%20FX%20Parameters) 2)))
                        (bind _output_get_3 (Variables|HiddenPartCheck|Advanced|SetWeaponHiddenFoot _outparthidden_2))
                        (return _output_get_3)
                        (elif (Variables|HiddenPartCheck|Advanced|GetWeaponHiddenFoot)
                          (bind _outparthidden_3 (CallFunction|SetPartFXVisibility false true (LocalVariables|GetIdleVFXArray) true (Utilities|Array|Get(acopy) (Variables|HiddenPartCheck|Advanced|GetHidden%20FX%20Parameters) 2)))
                          (bind _output_get_4 (Variables|HiddenPartCheck|Advanced|SetWeaponHiddenFoot _outparthidden_3))
                          (return _output_get_4)
                          (else
                            (bind _output_get_5 (Variables|HiddenPartCheck|Advanced|SetWeaponHiddenFoot false))
                            (return _output_get_5)))))))))))))))

(fn UpdateAudioStatValue
  (params (AudioComponent :type Array%20of%20Audio%20Component%20Object%20References) (InName :type Name) (StatValue :type Integer)))

(fn ConstructionScript)
