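; Part modifier for the LittleEgg body cosmetic that makes the character's materials and effects react to elimination count.
; A watched kill stat raises the level, driving capped Reactive_Level material values and a timeline hue shift.
; Variant material and part indices filter which Body or Face parts spawn attached Niagara overlays, fed level, front-end and hidden-head state.
; In the front end, a preview timeline cycles the levels through the same effects path.
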(event EventOnPostResIn
  (bind _returnvalue (CustomCharacter|GetCharacterPartType))
  (bind _returnvalue_2 (CustomCharacter|GetAssociatedPlayerPawn))
  (|CallParentFunction)
  (Utilities|FlowControl|Sequence
    (:then_0
      (bind _index (CosmeticLoadout|GetVariantChannelActiveIndexforSlotActor "(TagName=\"LoadoutSlot.Character\")" _returnvalue_2 "(TagName=\"Cosmetics.Variant.Channel.Material\")" (Utilities|Array|MakeArray "(TagName=\"Cosmetics.Variant.Property.Mat1\")" "(TagName=\"Cosmetics.Variant.Property.Mat2\")" "(TagName=\"Cosmetics.Variant.Property.Mat3\")")))
      (Variables|Default|SetVariant_MatChannelIndex _index)
      (bind _index_2 (CosmeticLoadout|GetVariantChannelActiveIndexforSlotActor "(TagName=\"LoadoutSlot.Character\")" _returnvalue_2 "(TagName=\"Cosmetics.Variant.Channel.Parts\")" (Utilities|Array|MakeArray "(TagName=\"Cosmetics.Variant.Property.Stage1\")" "(TagName=\"Cosmetics.Variant.Property.Stage2\")" "(TagName=\"Cosmetics.Variant.Property.Stage3\")")))
      (Variables|Default|SetVariant_PartChannelIndex _index_2))
    (:then_1
      (Cosmetics|WatchCosmeticStat (CustomCharacter|GetAssociatedPlayerPawn) "(TagName=\"StatManager.AthenaKills\")"
        (:then
          (Utilities|FlowControl|Delay 0.1))
        (:OnStatChanged
          _))
      (Utilities|FlowControl|Gate true
        (:Exit
          (bind _result (Math|Integer|IncrementInt (Variables|Default|GetLittleEgg_Level))
            (:%20%20
              (Rendering|Material|SetScalarParameterValueonMaterials (CustomCharacter|GetSkeletalMeshForPartOnPlayerPawn) "Reactive_Level" (Math|Conversions|ToFloat(Integer) (Math|Integer|Min(Integer) _result 5)))
              (Utilities|FlowControl|Sequence
                (:then_0
                  _)
                (:then_1
                  ;[1 0.4389 0.01696 0.432] Reactive FX
                  (if (CallFunction|FXReactiveFilter _returnvalue) :!label _label
                    (switch Utilities|FlowControl|Switch|SwitchonEFortCustomPartType _returnvalue
                      (:Body
                        _)
                      (:Face
                        _))
                    (bind _data (Utilities|FlowControl|MergeExecutionwithData "/BRCosmetics/Characters/Player/Female/Medium/Bodies/F_MED_Little_Egg_Jacket/FX/NS_F_Med_LittleEgg_Jacket_ReactiveOverlay.NS_F_Med_LittleEgg_Jacket_ReactiveOverlay" "/BRCosmetics/Characters/Player/Female/Medium/Bodies/F_MED_Little_Egg_Jacket/FX/NS_F_Med_LittleEgg_Mask_ReactiveOverlay.NS_F_Med_LittleEgg_Mask_ReactiveOverlay")
                      (:Then
                        (bind _returnvalue_3 (PartModifier|SpawnSystem|SpawnSystemAttachedPartModifier _data (CustomCharacter|GetSkeletalMeshForPartOnPlayerPawn) (Utilities|Select :Index _returnvalue :Head "None" :Body "None" :Hat "None" :Backpack "None" :MiscOrTail "None" :Face "head" :Gameplay "None" :ExtraPart "None" :Gameplay2 "None"))
                          (:Success
                            (Niagara|SetNiagaraVariable(Int32) _returnvalue_3 "ReactiveIndex" (Math|Integer|Clamp(Integer) (- (Variables|Default|GetLittleEgg_Level) 2) 0 6))
                            (Niagara|SetNiagaraVariable(Bool) _returnvalue_3 "bIsFrontEnd" (CustomCosmeticModifier|IsinFrontEnd))
                            (Niagara|SetNiagaraVariable(Bool) _returnvalue_3 "bisHeadHidden" (Variables|HiddenPartCheck|GetbIsHeadHidden))))))))))))))
    (:then_2
      (if (CustomCosmeticModifier|IsinFrontEnd)
        (Timeline|FE_Preview PlayFromStart
          (:Level_Up
            (bind _output_get (Variables|Default|SetLittleEgg_Level (+ (Math|Integer|%25%28Integer%29 (Variables|Default|GetLittleEgg_Level) 5) 1)))
            (Rendering|Material|SetScalarParameterValueonMaterials (CustomCharacter|GetSkeletalMeshForPartOnPlayerPawn) "Reactive_Level" (Math|Conversions|ToFloat(Integer) _output_get))
            (Utilities|FlowControl|Sequence
              (:then_0
                _)
              (:then_1
                (goto _label))))))))
  (Rendering|Material|SetScalarParameterValueonMaterials (CustomCharacter|GetSkeletalMeshForPartOnPlayerPawn) "Reactive_TimeFromBP" (Utilities|Time|GetGameTimeinSeconds))
  (bind (_direction _hueshift_progress) (Timeline|Reactive_HueShift PlayFromStart)
    (:Update
      (Rendering|Material|SetScalarParameterValueonMaterials (CustomCharacter|GetSkeletalMeshForPartOnPlayerPawn) "Reactive_ColorShift" _hueshift_progress))
    (:Finished
      (Rendering|Material|SetScalarParameterValueonMaterials (CustomCharacter|GetSkeletalMeshForPartOnPlayerPawn) "Reactive_ColorShift" 0.0))))

(event EventOnDayPhaseChanged
  (params (CurrentDayPhase :type EFortDayPhase%20Enum :default "Morning") (PreviousDayPhase :type EFortDayPhase%20Enum :default "Morning") (bAtCreation :type Boolean)))

(event EventOnNewPlayerWeaponType
  (params (NewType :type EFortItemType%20Enum :default "WorldItem") (OldType :type EFortItemType%20Enum :default "WorldItem")))

(event EventOnWeaponChanged
  (params (NewWeapon :type Fort%20Weapon%20Object%20Reference) (OldWeapon :type Fort%20Weapon%20Object%20Reference)))

(event EventOnWeaponFired)

(event EventOnNewWeapons
  (params (WeaponsAsItems :type Array%20of%20Fort%20Item%20Object%20References)))

(event EventOnCurieStateAttached
  (params (Tag :type Gameplay%20Tag%20Structure%20%28by%20ref%29)))

(event EventOnCurieStateDetached
  (params (Tag :type Gameplay%20Tag%20Structure%20%28by%20ref%29)))

(event EventOnBeginSkydiving)

(event EventOnEndSkydiving)

(event EventOnEnterVehicle
  (params (VehicleEntered :type Fort%20Vehicle%20Interface%20Interface%20%28by%20ref%29)))

(event EventOnExitVehicle)

(event EventOnTargetingOrFirstPersonCameraChanged)

(event EventOnHitPawn
  (params (SelfActor :type Actor%20Object%20Reference) (OtherActor :type Actor%20Object%20Reference) (NormalImpulse :type Vector :default "0, 0, 0") (Hit :type Hit%20Result%20Structure%20%28by%20ref%29)))

(fn FX_ReactiveFilter
  (params (InPartType :type EFortCustomPartType%20Enum :default "Head"))
  (bind _variant_partchannelindex (Variables|Default|GetVariant_PartChannelIndex))
  (bind _returnvalue (== _variant_partchannelindex 1))
  (switch Utilities|FlowControl|Switch|SwitchonEFortCustomPartType InPartType
    (:Body
      (return _returnvalue))
    (:Hat
      _)
    (:Backpack
      _)
    (:MiscOrTail
      _)
    (:Face
      (return (and (== (Variables|Default|GetVariant_MatChannelIndex) 2) _returnvalue)))
    (:Gameplay
      _))
  (return false))

(fn ConstructionScript)
