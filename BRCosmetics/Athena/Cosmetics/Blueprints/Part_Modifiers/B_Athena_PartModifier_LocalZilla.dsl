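; LocalZilla cosmetic part modifier that tones down the jacket's white variant in the front end so it does not look blown out.
; On post-res-in, where cosmetic events can run, it checks for front end and grabs the player part mesh's first material as the part MID.
; It compares that MID's parent to the resolved target material to adjust silk brightness and fresnel parameters.
; Other event overrides are empty or disabled.
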(event EventOnPostResIn
  (bind _targetmaterial (Variables|Default|GetTargetMaterial))
  (bind _output (Utilities|ResolveSoftReference _targetmaterial))
  (bind _partmid (Variables|Default|GetPartMID))
  (|CallParentFunction)
  (Networking|CanExecuteCosmeticEvents
    (:True
      (if (CustomCosmeticModifier|IsinFrontEnd)
        (Variables|Default|SetPartMID (Utilities|Casting|CastToMaterialInstanceDynamic (Rendering|Material|GetMaterial (CustomCharacter|GetSkeletalMeshForPartOnPlayerPawn) 0)))
        (Utilities|IsValid _output
          (:Is%20Valid
            (if (== _output (Variables|/Script/Engine.MaterialInstance|GetParent (Variables|Default|GetPartMID)))
              (Rendering|Material|SetScalarParameterValue _partmid "SilkBaseColorBrightness" 1.2 :!class /Script/Engine.MaterialInstanceDynamic) ; Reduce jacket brightness on white variant for frontend only to avoid blown out appearance
              (Rendering|Material|SetScalarParameterValue _partmid "SIlkFresnelMax" 0.2 :!class /Script/Engine.MaterialInstanceDynamic))))))))

(event EventOnDayPhaseChanged
  (params (CurrentDayPhase :type EFortDayPhase%20Enum :default "Morning") (PreviousDayPhase :type EFortDayPhase%20Enum :default "Morning") (bAtCreation :type Boolean)))

(event EventOnNewPlayerWeaponType
  (params (NewType :type EFortItemType%20Enum :default "WorldItem") (OldType :type EFortItemType%20Enum :default "WorldItem"))
  (|CallParentFunction NewType OldType) ; This node is disabled and will not be called.\nDrag off pins to build functionality.
)

(event EventOnWeaponChanged
  (params (NewWeapon :type Fort%20Weapon%20Object%20Reference) (OldWeapon :type Fort%20Weapon%20Object%20Reference)))

(event EventOnWeaponFired)

(event EventOnNewWeapons
  (params (WeaponsAsItems :type Array%20of%20Fort%20Item%20Object%20References)))

(event EventOnCurieStateAttached
  (params (Tag :type Gameplay%20Tag%20Structure%20%28by%20ref%29)))

(event EventOnCurieStateDetached
  (params (Tag :type Gameplay%20Tag%20Structure%20%28by%20ref%29)))

(event EventOnBeginSkydiving
  (|CallParentFunction) ; This node is disabled and will not be called.\nDrag off pins to build functionality.
)

(event EventOnEndSkydiving
  (|CallParentFunction) ; This node is disabled and will not be called.\nDrag off pins to build functionality.
)

(event EventOnEnterVehicle
  (params (VehicleEntered :type Fort%20Vehicle%20Interface%20Interface%20%28by%20ref%29))
  (|CallParentFunction VehicleEntered) ; This node is disabled and will not be called.\nDrag off pins to build functionality.
)

(event EventOnExitVehicle
  (|CallParentFunction) ; This node is disabled and will not be called.\nDrag off pins to build functionality.
)

(event EventOnTargetingOrFirstPersonCameraChanged)

(event EventOnHitPawn
  (params (SelfActor :type Actor%20Object%20Reference) (OtherActor :type Actor%20Object%20Reference) (NormalImpulse :type Vector :default "0, 0, 0") (Hit :type Hit%20Result%20Structure%20%28by%20ref%29)))

(fn ConstructionScript
  (|CallParentFunction))
