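; Character part modifier for the LoneDice cosmetic that makes it react to the player's eliminations.
; After post-res-in, it either plays a frontend preview timeline or watches the AthenaKills stat to trigger a reactive timeline and kill sound.
; Both timelines drive the LD_Reactivity material scalar and, for backpacks, Niagara Reactivity, zeroed unless the first material variant is active.
; The Sound Library audio call is a workaround for character part reuse.
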(event EventOnPostResIn
  (bind _returnvalue (CustomCharacter|GetCharacterPartType))
  (bind _returnvalue_2 (CustomCharacter|GetSkeletalMeshForPartOnPlayerPawn))
  (bind _returnvalue_3 (CallFunction|GetIdleVFXComponent))
  (|CallParentFunction)
  (if (Reactive|Frontend|ShouldDisplayFrontendReactivity 0 (CustomCharacter|GetCharacterPartType))
    (bind (_direction _reactivity) (Timeline|LD_FE_Preview PlayFromStart)
      (:Update
        (Rendering|Material|SetScalarParameterValueonMaterials _returnvalue_2 "LD_Reactivity" _reactivity)
        (bind _asniagara_particle_system_component (Utilities|Casting|CastToNiagaraComponent _returnvalue_3))
        (switch Utilities|FlowControl|Switch|SwitchonEFortCustomPartType _returnvalue
          (:Backpack
            (bind _index (CosmeticLoadout|GetVariantChannelActiveIndexforSlotActor "(TagName=\"LoadoutSlot.Backpack\")" (CustomCharacter|GetAssociatedPlayerPawn) "(TagName=\"Cosmetics.Variant.Channel.Material\")" (Utilities|Array|MakeArray "(TagName=\"Cosmetics.Variant.Property.Mat0\")" "(TagName=\"Cosmetics.Variant.Property.Mat1\")")))
            (if (== _index 0)
              (Niagara|SetNiagaraVariable(Float) _asniagara_particle_system_component "Reactivity" _reactivity)
              (else
                (Niagara|SetNiagaraVariable(Float) _asniagara_particle_system_component "Reactivity" 0.0)))))))
    (else
      (Cosmetics|WatchCosmeticStat (CustomCharacter|GetAssociatedPlayerPawn) "(TagName=\"StatManager.AthenaKills\")" 0 false
        (:OnStatChanged
          (CallFunction|PlayCustomAudio)
          (bind (_direction_2 _reactivity_2) (Timeline|LD_Reactive PlayFromStart)
            (:Update
              (Rendering|Material|SetScalarParameterValueonMaterials _returnvalue_2 "LD_Reactivity" _reactivity_2)
              (bind _asniagara_particle_system_component_2 (Utilities|Casting|CastToNiagaraComponent _returnvalue_3))
              (switch Utilities|FlowControl|Switch|SwitchonEFortCustomPartType _returnvalue
                (:Backpack
                  (bind _index_2 (CosmeticLoadout|GetVariantChannelActiveIndexforSlotActor "(TagName=\"LoadoutSlot.Backpack\")" (CustomCharacter|GetAssociatedPlayerPawn) "(TagName=\"Cosmetics.Variant.Channel.Material\")" (Utilities|Array|MakeArray "(TagName=\"Cosmetics.Variant.Property.Mat0\")" "(TagName=\"Cosmetics.Variant.Property.Mat1\")")))
                  (if (== _index_2 0)
                    (Niagara|SetNiagaraVariable(Float) _asniagara_particle_system_component_2 "Reactivity" _reactivity_2)
                    (else
                      (Niagara|SetNiagaraVariable(Float) _asniagara_particle_system_component_2 "Reactivity" 0.0))))))))))))

(event Custom|PlayCustomAudio
  (bind _returnvalue (CustomCharacter|GetSkeletalMeshForPartOnPlayerPawn))
  ;[0 0 0 0.5] Workaround for reactive audio due to the character part reuse
  (Networking|CanExecuteCosmeticEvents
    (:True
      (Utilities|IsValid _returnvalue
        (:Is%20Valid
          (SoundLibrary|PlaySoundSimple (WorldSubsystems|GetSoundLibrarySubsystem) self (Utilities|Struct|MakeSimpleSoundSettings "(TagName=\"SoundLibrary.Cosmetic.Kills\")" true false true true 1.0 1.0 0.0 0 0 "None" _returnvalue) :!class /Script/SoundLibrary.SoundLibrarySubsystem))))))

(fn ConstructionScript
  (|CallParentFunction))
