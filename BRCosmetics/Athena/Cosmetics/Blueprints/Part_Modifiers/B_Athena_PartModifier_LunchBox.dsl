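; Part modifier for the LunchBox character cosmetic that escalates its look through four stages as cosmetic states activate.
; Stages drive part Reveal and Face/Head emissive parameters, mostly via timelines, and activate attached stage three/four FX, with DoOnce gating the later stages.
; Positive stat changes spawn a body Niagara effect that receives the hand-hidden flag.
; In the frontend, a looping timer cycles the stages as a preview.
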(event EventOnPostResIn
  (bind _returnvalue (CustomCharacter|GetAssociatedPlayerPawn))
  (bind _returnvalue_2 (CustomCharacter|GetCharacterPartType))
  (bind _returnvalue_3 (Customization|GetSkeletalMeshforPartType _returnvalue _returnvalue_2 :!class /Script/FortniteGame.FortPlayerPawn))
  (bind _additional_fx_needing_part_checks (Variables|HiddenPartCheck|GetAdditional%20FX%20Needing%20Part%20Checks))
  (|CallParentFunction)
  (Transformation|AttachComponentToComponent (Variables|Default|GetStageThreeFX) _returnvalue_3 "root" "KeepRelative" "KeepRelative" "KeepRelative" true)
  (Transformation|AttachComponentToComponent (Variables|Default|GetStageFourFX) _returnvalue_3 "root" "KeepRelative" "KeepRelative" "KeepRelative" true)
  (Utilities|Array|Add _additional_fx_needing_part_checks (Variables|Default|GetStageThreeFX))
  (Utilities|Array|Add _additional_fx_needing_part_checks (Variables|Default|GetStageFourFX))
  (if (CustomCosmeticModifier|IsinFrontEnd)
    (Utilities|Time|SetTimerbyEvent (EventDispatchers|CreateEvent :!event Custom|FrontendStageStack) 5.0 true)))

(event Custom|FrontendStageStack
  (switch int (Variables|Default|GetFE_Stage)
    (:0
      (CallFunction|StageOne))
    (:1
      (CallFunction|FEChanged)
      (CallFunction|StageTwo))
    (:2
      (CallFunction|FEChanged)
      (CallFunction|StageThree))
    (:3
      (CallFunction|FEChanged)
      (CallFunction|StageFour)
      (Variables|Default|SetFE_Stage 0)))
  (Math|Integer|IncrementInt (Variables|Default|GetFE_Stage)))

(event EventOnCosmeticStateActivated
  (params (Stat%20Name :type Name :default "None") (Stat%20Value :type Integer) (STAGE%20Num%20Being%20Activated :type Integer))
  (|CallParentFunction Stat%20Name Stat%20Value STAGE%20Num%20Being%20Activated)
  (switch int STAGE%20Num%20Being%20Activated
    (:0
      (CallFunction|StageOne))
    (:1
      (CallFunction|StageTwo))
    (:2
      (CallFunction|StageThree))
    (:3
      (CallFunction|StageFour))))

(event Custom|FE_Changed
  (Utilities|FlowControl|Delay 0.1)
  (bind _returnvalue (PartModifier|SpawnSystem|SpawnSystemAttachedPartModifier "/BRCosmetics/Effects/Fort_Effects/Effects/Characters/Athena_Parts/LunchBox/NS_LunchBox_Body.NS_LunchBox_Body" (Customization|GetSkeletalMeshforPartType (CustomCharacter|GetAssociatedPlayerPawn) (CustomCharacter|GetCharacterPartType) :!class /Script/FortniteGame.FortPlayerPawn) "root")
    (:Success
      (Effects|Components|ParticleSystem|SetBoolParameter _returnvalue "bisHandHidden" (Variables|HiddenPartCheck|GetbIsHandHidden)))))

(event EventOnStatChanged
  (params (State%20Name :type Name :default "None") (Stat%20Value :type Integer))
  (|CallParentFunction State%20Name Stat%20Value)
  (if (> Stat%20Value 0)
    (Utilities|FlowControl|Delay 0.1)
    (bind _returnvalue (PartModifier|SpawnSystem|SpawnSystemAttachedPartModifier "/BRCosmetics/Effects/Fort_Effects/Effects/Characters/Athena_Parts/LunchBox/NS_LunchBox_Body.NS_LunchBox_Body" (Customization|GetSkeletalMeshforPartType (CustomCharacter|GetAssociatedPlayerPawn) (CustomCharacter|GetCharacterPartType) :!class /Script/FortniteGame.FortPlayerPawn) "root")
      (:Success
        (Effects|Components|ParticleSystem|SetBoolParameter _returnvalue "bisHandHidden" (Variables|HiddenPartCheck|GetbIsHandHidden))))))

(event Custom|StageOne
  (bind _returnvalue (CustomCharacter|GetAssociatedPlayerPawn))
  (bind _returnvalue_2 (Customization|GetSkeletalMeshforPartType _returnvalue "Head" :!class /Script/FortniteGame.FortPlayerPawn))
  ;[1 1 1] Stage One
  (Rendering|Material|SetScalarParameterValueonMaterials (CustomCharacter|GetSkeletalMeshForPartOnPlayerPawn) "Reveal" 0.0)
  (Rendering|Material|SetScalarParameterValueonMaterials (Customization|GetSkeletalMeshforPartType (CustomCharacter|GetAssociatedPlayerPawn) "Face" :!class /Script/FortniteGame.FortPlayerPawn) "Emissive Mult" 0.0)
  (Rendering|Material|SetScalarParameterValueonMaterials _returnvalue_2 "Emissive_BrightnessMax" 0.0)
  (Rendering|Material|SetScalarParameterValueonMaterials _returnvalue_2 "Emissive_BrightnessMin" 0.0)
  (Components|Activation|Deactivate (Variables|Default|GetStageThreeFX))
  (CallFunction|RESET))

(event Custom|StageTwo
  ;[1 1 1] Stage Two
  (bind (_direction _newtrack_0) (Timeline|Timeline_0 PlayFromStart)
    (:Update
      (Rendering|Material|SetScalarParameterValueonMaterials (CustomCharacter|GetSkeletalMeshForPartOnPlayerPawn) "Reveal" (Math|Float|Lerp 0.0 0.6 _newtrack_0))
      (Rendering|Material|SetScalarParameterValueonMaterials (Customization|GetSkeletalMeshforPartType (CustomCharacter|GetAssociatedPlayerPawn) "Face" :!class /Script/FortniteGame.FortPlayerPawn) "Emissive Mult" (Math|Float|Lerp 0.0 7.0 _newtrack_0)))))

(event Custom|StageThree
  (bind _returnvalue (CustomCharacter|GetAssociatedPlayerPawn))
  (bind _returnvalue_2 (Customization|GetSkeletalMeshforPartType _returnvalue "Head" :!class /Script/FortniteGame.FortPlayerPawn))
  ;[1 1 1] Stage Three
  (Utilities|FlowControl|DoOnce false
    (:Completed
      (Components|Activation|Activate (Variables|Default|GetStageThreeFX) false)
      (bind (_direction _newtrack_0) (Timeline|Timeline_1 PlayFromStart)
        (:Update
          (Rendering|Material|SetScalarParameterValueonMaterials (CustomCharacter|GetSkeletalMeshForPartOnPlayerPawn) "Reveal" (Math|Float|Lerp 0.0 0.9 _newtrack_0))
          (Rendering|Material|SetScalarParameterValueonMaterials _returnvalue_2 "Emissive_BrightnessMax" (Math|Float|Lerp 0.0 15.0 _newtrack_0))
          (Rendering|Material|SetScalarParameterValueonMaterials _returnvalue_2 "Emissive_BrightnessMin" (Math|Float|Lerp 0.0 7.0 _newtrack_0)))))))

(event Custom|RESET
  (bind _returnvalue (CustomCharacter|GetAssociatedPlayerPawn))
  (bind _returnvalue_2 (Customization|GetSkeletalMeshforPartType _returnvalue "Head" :!class /Script/FortniteGame.FortPlayerPawn))
  (Utilities|FlowControl|Sequence
    (:then_0
      ;[1 1 1] Stage Three
      (Utilities|FlowControl|DoOnce false
        (:Completed
          (Components|Activation|Activate (Variables|Default|GetStageThreeFX) false)
          (Timeline|Timeline_1 PlayFromStart))))
    (:then_1
      ;[1 1 1] Stage Four
      (Utilities|FlowControl|DoOnce false
        (:Completed
          (Components|Activation|Activate (Variables|Default|GetStageThreeFX) false)
          (Components|Activation|Activate (Variables|Default|GetStageFourFX) false)
          (Niagara|SetNiagaraVariable(Bool) (Variables|Default|GetStageThreeFX) "LightOn" false)
          (bind (_direction _newtrack_0) (Timeline|Timeline_2 PlayFromStart)
            (:Update
              (Rendering|Material|SetScalarParameterValueonMaterials (CustomCharacter|GetSkeletalMeshForPartOnPlayerPawn) "Reveal" (Math|Float|Lerp 0.0 1.0 _newtrack_0)))))))))

(event Custom|StageFour
  ;[1 1 1] Stage Four
  (Utilities|FlowControl|DoOnce false
    (:Completed
      (Components|Activation|Activate (Variables|Default|GetStageThreeFX) false)
      (Components|Activation|Activate (Variables|Default|GetStageFourFX) false)
      (Niagara|SetNiagaraVariable(Bool) (Variables|Default|GetStageThreeFX) "LightOn" false)
      (Timeline|Timeline_2 PlayFromStart))))

(fn ConstructionScript
  (|CallParentFunction))
