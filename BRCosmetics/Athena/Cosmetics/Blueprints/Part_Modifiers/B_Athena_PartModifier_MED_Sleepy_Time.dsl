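; Cosmetic part modifier for the Sleepy Time character, attaching six glow meshes to the player's limbs.
; When customization finishes, MeshSetup snaps each mesh to an arm or thigh twist socket on the part's skeletal mesh, applying rotation, rotated local offset and scale.
; The first (left forearm) mesh follows the hand-hidden check on part application and weapon change, and is hidden in game at Medium effects quality.
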(event EventOnCosmeticPartAppliedPostResIn
  (params (Part :type Custom%20Character%20Part%20Object%20Reference))
  (|CallParentFunction Part)
  (if (Variables|HiddenPartCheck|GetbIsHandHidden)
    (Rendering|SetVisibility (Variables|Default|GetBluGlowModel_NoLOD) false true)
    (else
      (Rendering|SetVisibility (Variables|Default|GetBluGlowModel_NoLOD) true true))))

(event EventOnWeaponChanged
  (params (NewWeapon :type Fort%20Weapon%20Object%20Reference) (OldWeapon :type Fort%20Weapon%20Object%20Reference))
  (|CallParentFunction NewWeapon OldWeapon)
  (if (Variables|HiddenPartCheck|GetbIsHandHidden)
    (Rendering|SetVisibility (Variables|Default|GetBluGlowModel_NoLOD) false true)
    (else
      (Rendering|SetVisibility (Variables|Default|GetBluGlowModel_NoLOD) true true))))

(event EventOnCharacterCustomizationFinished
  (params (CharacterPartInfo :type Array%20of%20Applied%20Character%20Part%20Info%20Structures))
  (|CallParentFunction CharacterPartInfo)
  (Utilities|EffectsQualityCheck 0
    (:Cinematic
      _)
    (:Epic
      _)
    (:High
      _)
    (:Medium
      (Development|SetHiddeninGame (Variables|Default|GetBluGlowModel_NoLOD) true false))
    (:Low
      _))
  (CallFunction|MeshSetup (Variables|Default|GetBluGlowModel_NoLOD) "lowerarm_twist_02_l" "90.000000, 0, 0" "0, 0,-7.500000")
  (CallFunction|MeshSetup (Variables|Default|GetBluGlowModel_NoLOD1) "upperarm_twist_02_l" "90.000000, 0, 0" "0, 0,-12.000000")
  (CallFunction|MeshSetup (Variables|Default|GetBluGlowModel_NoLOD2) "lowerarm_twist_01_r" "270.000000, 0,-4.000000")
  (CallFunction|MeshSetup (Variables|Default|GetBluGlowModel_NoLOD3) "upperarm_twist_02_r" "270.000000, 0, 0" "0, 0,-12.000000")
  (CallFunction|MeshSetup (Variables|Default|GetBluGlowModel_NoLOD4) "thigh_twist_01_r" "90.000000, 0, 0" "0, 0,-12.000000" "1.000000,1.000000,2.000000")
  (CallFunction|MeshSetup (Variables|Default|GetBluGlowModel_NoLOD5) "thigh_twist_01_l" "270.000000, 0, 0" "0, 0,-12.000000" "1.000000,1.000000,2.000000"))

(event EventOnPostResIn
  (|CallParentFunction))

(fn MeshSetup
  (params (Mesh :type Static%20Mesh%20Component%20Object%20Reference) (SocketName :type Name) (Relative%20Rotation :type Rotator) (Local%20Offset :type Vector) (Relative%20Scale%203D :type Vector :default "1.000000,1.000000,1.000000"))
  (Transformation|AttachComponentToComponent Mesh (CustomCharacter|GetSkeletalMeshForPartOnPlayerPawn) SocketName "SnapToTarget" "KeepRelative" "SnapToTarget" true)
  (Transformation|SetRelativeRotation (Variables|GetMesh) (Variables|GetRelative%20Rotation) false false)
  (Transformation|SetRelativeLocation (Variables|GetMesh) (Math|Transform|TransformDirection (Transformation|GetRelativeTransform (Variables|GetMesh)) (Variables|GetLocal%20Offset)) false false)
  (Transformation|SetRelativeScale3D (Variables|GetMesh) (Variables|GetRelative%20Scale%203D)))

(fn ConstructionScript)
