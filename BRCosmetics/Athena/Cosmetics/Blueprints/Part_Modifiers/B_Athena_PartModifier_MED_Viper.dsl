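; Part modifier blueprint for the MED Viper Athena cosmetic that tells its idle VFX whether the pawn is jumping.
; The jumped and landed events each call the parent implementation, then get the idle VFX particle component.
; They set its bool parameter bisJumping to true on jump and false on landing.
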(event EventOnPawnJumped
  (|CallParentFunction)
  (Effects|Components|ParticleSystem|SetBoolParameter (CallFunction|GetIdleVFXComponent) "bisJumping" true))

(event EventOnPawnLanded
  (|CallParentFunction)
  (Effects|Components|ParticleSystem|SetBoolParameter (CallFunction|GetIdleVFXComponent) "bisJumping" false))

(fn ConstructionScript)
