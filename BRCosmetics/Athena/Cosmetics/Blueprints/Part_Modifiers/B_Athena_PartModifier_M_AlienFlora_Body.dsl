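; Part modifier Blueprint for the Alien Flora medium male body cosmetic, adding a fallen-leaves effect when the player pawn lands.
; On the landed event it calls the parent, then spawns the NS_Skin_AlienFlora_FallenLeaves Niagara system attached to the root of the part's skeletal mesh on the player pawn.
; The effect snaps to that root, including scale, and the construction script is empty.
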(event EventOnPawnLanded
  (|CallParentFunction)
  (PartModifier|SpawnSystem|SpawnSystemAttachedPartModifier "/BRCosmetics/Characters/Player/Male/Medium/Bodies/M_MED_Alien_Flora/FX/NS_Skin_AlienFlora_FallenLeaves.NS_Skin_AlienFlora_FallenLeaves" (CustomCharacter|GetSkeletalMeshForPartOnPlayerPawn) "root" "0, 0, 0" "0, 0, 0" "SnapToTargetIncludingScale"))

(fn ConstructionScript)
