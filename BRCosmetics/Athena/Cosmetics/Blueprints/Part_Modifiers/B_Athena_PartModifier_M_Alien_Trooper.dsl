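; Cosmetic part modifier for the male Alien Trooper outfit that swaps the face textures to match the selected style variants.
; After res-in, it reads the active Progressive color and Pattern material variant indices from the character loadout slot.
; These form a flattened 7x7 row index (forehead * 7 + pattern) into a texture data table.
; The row's diffuse, FX mask and normal textures go onto dynamic materials of the Face mesh, skipping CyclopsEye materials.
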(event EventOnPostResIn
  (bind _returnvalue (CustomCharacter|GetAssociatedPlayerPawn))
  (|CallParentFunction)
  (bind _index (CosmeticLoadout|GetVariantChannelActiveIndexforSlotActor "(TagName=\"LoadoutSlot.Character\")" _returnvalue "(TagName=\"Cosmetics.Variant.Channel.Progressive\")" (Utilities|Array|MakeArray "(TagName=\"Cosmetics.Variant.Property.Color.000\")" "(TagName=\"Cosmetics.Variant.Property.Color.001\")" "(TagName=\"Cosmetics.Variant.Property.Color.002\")" "(TagName=\"Cosmetics.Variant.Property.Color.003\")" "(TagName=\"Cosmetics.Variant.Property.Color.004\")" "(TagName=\"Cosmetics.Variant.Property.Color.005\")" "(TagName=\"Cosmetics.Variant.Property.Color.006\")")))
  (bind _index_2 (CosmeticLoadout|GetVariantChannelActiveIndexforSlotActor "(TagName=\"LoadoutSlot.Character\")" _returnvalue "(TagName=\"Cosmetics.Variant.Channel.Pattern\")" (Utilities|Array|MakeArray "(TagName=\"Cosmetics.Variant.Property.Mat0\")" "(TagName=\"Cosmetics.Variant.Property.Mat1\")" "(TagName=\"Cosmetics.Variant.Property.Mat2\")" "(TagName=\"Cosmetics.Variant.Property.Mat3\")" "(TagName=\"Cosmetics.Variant.Property.Mat4\")" "(TagName=\"Cosmetics.Variant.Property.Mat5\")" "(TagName=\"Cosmetics.Variant.Property.Mat6\")")))
  (CallFunction|SetFaceTextures _index _index_2))

(fn setFaceTextures
  (params (Forehead :type Integer) (Pattern :type Integer))
  (bind _returnvalue (CustomCharacter|GetAssociatedPlayerPawn))
  (bind _returnvalue_2 (Customization|GetSkeletalMeshforPartType _returnvalue "Face" :!class /Script/FortniteGame.FortPlayerPawn))
  (bind _outrownames (DataTable|GetDataTableRowNames "/BRCosmetics/Characters/Player/Male/Medium/Bodies/M_MED_Alien_Trooper/BluePrint/AlienTrooperTextureDataTable.AlienTrooperTextureDataTable"))
  (bind _returnvalue_3 (Utilities|GetDataTableRow "/BRCosmetics/Characters/Player/Male/Medium/Bodies/M_MED_Alien_Trooper/BluePrint/AlienTrooperTextureDataTable.AlienTrooperTextureDataTable" (Utilities|Array|Get(acopy) _outrownames (+ (* Forehead 7) Pattern))))
  (bind (_returnvalue_at_diffuse_7_a295119d474b4fe1fb8246bd56d01953 _returnvalue_at_fxmask_8_bd567d214e016e89dc545ba06b489d8e _returnvalue_normal_11_0c7ee87d4b46fbaa7bb62c8c28ea934b) (Utilities|Struct|BreakAlientTrooperTextureStructure _returnvalue_3))
  (for (_array_element _array_index) (Rendering|Material|GetMaterials _returnvalue_2)
    (if (not (Utilities|String|Contains (Utilities|GetDisplayName (Variables|/Script/Engine.MaterialInstance|GetParent (Material|GetMIDfromMaterial _returnvalue_2 _array_index _array_element))) "CyclopsEye"))
      (Rendering|Material|SetTextureParameterValue (Material|GetMIDfromMaterial _returnvalue_2 _array_index _array_element) "AT_Diffuse" _returnvalue_at_diffuse_7_a295119d474b4fe1fb8246bd56d01953)
      (Rendering|Material|SetTextureParameterValue (Material|GetMIDfromMaterial _returnvalue_2 _array_index _array_element) "AT_FXMask" _returnvalue_at_fxmask_8_bd567d214e016e89dc545ba06b489d8e)
      (Rendering|Material|SetTextureParameterValue (Material|GetMIDfromMaterial _returnvalue_2 _array_index _array_element) "Normals" _returnvalue_normal_11_0c7ee87d4b46fbaa7bb62c8c28ea934b))))

(fn ConstructionScript)
