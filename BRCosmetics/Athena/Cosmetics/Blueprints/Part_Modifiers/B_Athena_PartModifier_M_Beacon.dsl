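; Cosmetic part modifier for the Beacon character that shows eye-beam effects at night and in the front end.
; After res-in it attaches both beam components to spine_05 on the part's skeletal mesh, then sets the isNight flag and visibility, which day-phase changes also update.
; Showing requires an unhidden head, rechecked on weapon or part changes during night.
; As written, show calls target only EyeBeamR, while day and non-front-end hide calls target only EyeBeamL.
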(event EventOnPostResIn
  (bind _returnvalue (CustomCharacter|GetSkeletalMeshForPartOnPlayerPawn))
  (|CallParentFunction)
  (Utilities|FlowControl|Sequence
    (:then_0
      (Transformation|AttachComponentToComponent (Variables|Default|GetEyeBeamR) _returnvalue "spine_05" "KeepRelative" "KeepRelative" "KeepRelative" true)
      (Transformation|AttachComponentToComponent (Variables|Default|GetEyeBeamL) _returnvalue "spine_05" "KeepRelative" "KeepRelative" "KeepRelative" true))
    (:then_1
      (if (CustomCosmeticModifier|IsinFrontEnd)
        (if (not (Variables|HiddenPartCheck|GetbIsHeadHidden))
          (Variables|Default|SetisNight true)
          (Rendering|SetVisibility (Variables|Default|GetEyeBeamR) true false))
        (else
          (Variables|Default|SetisNight false)
          (Rendering|SetVisibility (Variables|Default|GetEyeBeamL) false false))))))

(event EventOnDayPhaseChanged
  (params (CurrentDayPhase :type EFortDayPhase%20Enum :default "Morning") (PreviousDayPhase :type EFortDayPhase%20Enum :default "Morning") (bAtCreation :type Boolean))
  (|CallParentFunction CurrentDayPhase PreviousDayPhase bAtCreation)
  (if (Utilities|Enum|Equal(Enum) CurrentDayPhase "Night")
    (if (not (Variables|HiddenPartCheck|GetbIsHeadHidden))
      (Variables|Default|SetisNight true)
      (Rendering|SetVisibility (Variables|Default|GetEyeBeamR) true false))
    (else
      (Variables|Default|SetisNight false)
      (Rendering|SetVisibility (Variables|Default|GetEyeBeamL) false false))))

(event EventOnWeaponChanged
  (params (NewWeapon :type Fort%20Weapon%20Object%20Reference) (OldWeapon :type Fort%20Weapon%20Object%20Reference))
  (bind _eyebeamr (Variables|Default|GetEyeBeamR))
  (|CallParentFunction NewWeapon OldWeapon)
  (if (Variables|HiddenPartCheck|GetbIsHeadHidden)
    (if (Variables|Default|GetisNight)
      (Rendering|SetVisibility _eyebeamr false false))
    (else
      (if (Variables|Default|GetisNight)
        (Rendering|SetVisibility _eyebeamr true false)))))

(event EventOnCosmeticPartAppliedPostResIn
  (params (Part :type Custom%20Character%20Part%20Object%20Reference))
  (bind _eyebeamr (Variables|Default|GetEyeBeamR))
  (|CallParentFunction Part)
  (if (Variables|HiddenPartCheck|GetbIsHeadHidden)
    (if (Variables|Default|GetisNight)
      (Rendering|SetVisibility _eyebeamr false false))
    (else
      (if (Variables|Default|GetisNight)
        (Rendering|SetVisibility _eyebeamr true false)))))

(fn ConstructionScript)
