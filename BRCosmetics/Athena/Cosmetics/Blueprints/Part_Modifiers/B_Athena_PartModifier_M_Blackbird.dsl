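; Cosmetic part modifier for the Blackbird outfit that drives "SuperLevelling" material parameters on the part's skeletal mesh from the player's season level stat.
; On post-res-in it watches SeasonLevel12, keeps the highest value seen, and pushes level bounds and current level, backed by a short-lived looping refresh timer workaround.
; In the front end it also sets LUT texture parameters on each material.
; New weapons and entered vehicles receive an override wrap when one is set.
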(event Custom|HackyWorkaroundRefreshTimerGuy
  ;[1 1 1] SuperLevelling Logic
  (CallFunction|InitializeVariabels)
  (Rendering|Material|SetScalarParameterValueonMaterials (CustomCharacter|GetSkeletalMeshForPartOnPlayerPawn) "SL_CurrentLevel" (Math|Conversions|ToFloat(Integer) (stale EventOnPostResIn _output_get)))
  (Utilities|FlowControl|DoOnce false
    (:Completed
      (bind _returnvalue (Utilities|Time|SetTimerbyEvent (delegate Custom|HackyWorkaroundRefreshTimerGuy) 0.1 true))
      (Variables|Default|SetSL_RefreshTimeHandler _returnvalue))))

(event Custom|ResetSafetyGate
  ;[1 1 1] SuperLevelling Logic
  (Utilities|FlowControl|DoOnce false
    (:Completed
      (bind _returnvalue (Utilities|Time|SetTimerbyEvent (delegate Custom|HackyWorkaroundRefreshTimerGuy) 0.1 true))
      (Variables|Default|SetSL_RefreshTimeHandler _returnvalue))))

(event EventOnPostResIn
  (bind _returnvalue (CustomCharacter|GetSkeletalMeshForPartOnPlayerPawn))
  (bind _returnvalue_2 (CustomCharacter|GetSkeletalMeshForPartOnPlayerPawn))
  (bind _is_material_instance (Variables|Default|GetIs%20Material%20Instance))
  (|CallParentFunction)
  (CallFunction|ResetSafetyGate)
  ;[1 1 1] SuperLevelling Logic
  (Utilities|FlowControl|Sequence
    (:then_0
      (Utilities|FlowControl|Delay 1.0)
      (Utilities|Time|ClearandInvalidateTimerbyHandle (Variables|Default|GetSL_RefreshTimeHandler)))
    (:then_1
      (bind (_statname _statvalue) (Cosmetics|WatchCosmeticStat (CustomCharacter|GetAssociatedPlayerPawn) "(TagName=\"StatManager.Season.SeasonLevel12\")")
        (:OnStatChanged
          (Rendering|Material|SetScalarParameterValueonMaterials _returnvalue_2 "SL_StartAtLevel" (Math|Conversions|ToFloat(Integer) (Variables|Default|GetSL_StartLevel)))
          (Rendering|Material|SetScalarParameterValueonMaterials _returnvalue_2 "SL_StopAtLevel" (Math|Conversions|ToFloat(Integer) (Variables|Default|GetSL_EndLevel)))
          (Rendering|Material|SetScalarParameterValueonMaterials _returnvalue_2 "SL_EndRadius" (Variables|Default|GetEndRadius))
          (bind _output_get (Variables|Default|SetSL_LastValue (Math|Integer|Max(Integer) _statvalue (Variables|Default|GetSL_LastValue))))
          (CallFunction|InitializeVariabels)
          (Rendering|Material|SetScalarParameterValueonMaterials (CustomCharacter|GetSkeletalMeshForPartOnPlayerPawn) "SL_CurrentLevel" (Math|Conversions|ToFloat(Integer) _output_get))
          (Utilities|FlowControl|DoOnce false
            (:Completed
              (bind _returnvalue_3 (Utilities|Time|SetTimerbyEvent (delegate Custom|HackyWorkaroundRefreshTimerGuy) 0.1 true))
              (Variables|Default|SetSL_RefreshTimeHandler _returnvalue_3))))))
    (:then_2
      ;[1 1 1] Front End LUT, Remove after creating auto LUT tool
      (if (CustomCosmeticModifier|IsinFrontEnd)
        (for (_array_element _array_index) (Rendering|Material|GetMaterials _returnvalue)
          (bind _asmaterial_instance_dynamic (Utilities|Casting|CastToMaterialInstanceDynamic _array_element)
            (:then
              (Variables|Default|SetIs%20Material%20Instance true))
            (:CastFailed
              (bind _returnvalue_4 (Rendering|Material|CreateDynamicMaterialInstance _returnvalue _array_index _array_element "None" :!class /Script/Engine.PrimitiveComponent))
              (Variables|Default|SetIs%20Material%20Instance false)))
          (Rendering|Material|SetTextureParameterValue (select _is_material_instance _asmaterial_instance_dynamic _returnvalue_4) "LUT Volume Texture" "/Game/Packages/Fortress_SharedMaterials/Anime_Master_Material/VT_InverseLUT_FE.VT_InverseLUT_FE")
          (Rendering|Material|SetScalarParameterValue (select _is_material_instance _asmaterial_instance_dynamic _returnvalue_4) "LUT Multiplier" 1.0 :!class /Script/Engine.MaterialInstanceDynamic)
          (Rendering|Material|SetScalarParameterValue (select _is_material_instance _asmaterial_instance_dynamic _returnvalue_4) "Mobile LUT Brightness" 1.0 :!class /Script/Engine.MaterialInstanceDynamic)
          (Rendering|Material|SetScalarParameterValue (select _is_material_instance _asmaterial_instance_dynamic _returnvalue_4) "use 33 Resolution" 0.0 :!class /Script/Engine.MaterialInstanceDynamic))))))

(event EventOnNewWeapons
  (params (WeaponsAsItems :type Array%20of%20Fort%20Item%20Object%20References))
  (bind _pendingitemstowrap (Variables|Default|GetPendingItemsToWrap))
  (for _array_element WeaponsAsItems
    (switch Utilities|FlowControl|Switch|SwitchonEFortItemType (Item|GetType _array_element)
      (:AthenaGadget
        _)
      (:WeaponMelee
        _)
      (:WeaponRanged
        _))
    (Utilities|Array|Add (Variables|Default|GetPendingItemsToWrap) _array_element))
  (if (Utilities|IsValidSoftObjectReference (Variables|Default|GetWeaponWrapToUse))
    (CallFunction|ApplyOverridetoItems _pendingitemstowrap)
    (Utilities|Array|Clear _pendingitemstowrap)))

(event Custom|InitializeVariabels
  (bind _returnvalue (CustomCharacter|GetSkeletalMeshForPartOnPlayerPawn))
  (Rendering|Material|SetScalarParameterValueonMaterials _returnvalue "SL_StartAtLevel" (Math|Conversions|ToFloat(Integer) (Variables|Default|GetSL_StartLevel)))
  (Rendering|Material|SetScalarParameterValueonMaterials _returnvalue "SL_StopAtLevel" (Math|Conversions|ToFloat(Integer) (Variables|Default|GetSL_EndLevel)))
  (Rendering|Material|SetScalarParameterValueonMaterials _returnvalue "SL_EndRadius" (Variables|Default|GetEndRadius)))

(event EventOnEnterVehicle
  (params (VehicleEntered :type Fort%20Vehicle%20Interface%20Interface%20%28by%20ref%29))
  (if (Utilities|IsValidSoftObjectReference (Variables|Default|GetWeaponWrapToUse))
    (bind _as_fort_player_controller (CallFunction|GetAssociatedController))
    (Utilities|IsValid _as_fort_player_controller
      (:Is%20Valid
        (Class|FortPlayerController|ApplyOverrideWraptoVehicle _as_fort_player_controller (Variables|Default|GetWeaponWrapToUse))))))

(fn ApplyLevelParam
  (params (Level :type Integer))
  (Rendering|Material|SetScalarParameterValueonMaterials (CustomCharacter|GetSkeletalMeshForPartOnPlayerPawn) "SL_CurrentLevel" (Math|Conversions|ToFloat(Integer) Level)))

(fn ApplyOverrideToItems
  (params (Array :type Array%20of%20Fort%20Item%20Object%20References))
  (bind _as_fort_player_controller (CallFunction|GetAssociatedController))
  (Utilities|IsValid _as_fort_player_controller
    (:Is%20Valid
      (for _array_element Array
        (Class|FortPlayerController|ServerApplyOverrideWraptoItem _as_fort_player_controller (Item|GetItemGuid _array_element) (Variables|Default|GetWeaponWrapToUse))))))

(fn GetAssociatedController
  (bind _returnvalue (CustomCharacter|GetAssociatedPlayerPawn))
  (Utilities|FlowControl|Sequence
    (:then_0
      (bind _aspawn (Utilities|Casting|CastToPawn (Vehicle|GetVehicleActor _returnvalue :!class /Script/FortniteGame.FortPlayerPawn)))
      (bind _asfort_player_controller (Utilities|Casting|CastToFortPlayerController (Pawn|GetController _aspawn)))
      (return _asfort_player_controller))
    (:then_1
      (bind _asfort_player_controller_2 (Utilities|Casting|CastToFortPlayerController (Pawn|GetController _returnvalue))
        (:then
          _)
        (:CastFailed
          _))
      (return _asfort_player_controller_2))))

(fn ConstructionScript
  (bind _returnvalue (CustomCharacter|GetSkeletalMeshForPartOnPlayerPawn))
  (|CallParentFunction)
  (Rendering|Material|SetScalarParameterValueonMaterials (CustomCharacter|GetSkeletalMeshForPartOnPlayerPawn) "SL_GameTimeFromBP" -99999.0)
  (Rendering|Material|SetScalarParameterValueonMaterials _returnvalue "SL_StartAtLevel" (Math|Conversions|ToFloat(Integer) (Variables|Default|GetSL_StartLevel)))
  (Rendering|Material|SetScalarParameterValueonMaterials _returnvalue "SL_StopAtLevel" (Math|Conversions|ToFloat(Integer) (Variables|Default|GetSL_EndLevel))))
