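; Cosmetic part modifier for the Athena BuffCat character part, driving super-level material visuals and kill-reactive glow.
; After PostResIn it attaches Idle, Update and Top particles, then watches the season level stat to set SL_* mesh material scalars, keeping the highest value.
; A 0.1-second looping workaround timer re-pushes those scalars, with a clear scheduled one second after PostResIn.
; Kill stages raise GlowAlpha on Idle, then Top at stage five; stat changes activate Idle.
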(event Custom|HackyWorkaroundRefreshTimerGuy
  ;[1 1 1] SuperLevelling Logic
  (CallFunction|InitializeVariables)
  (Rendering|Material|SetScalarParameterValueonMaterials (CustomCharacter|GetSkeletalMeshForPartOnPlayerPawn) "SL_CurrentLevel" (Math|Conversions|ToFloat(Integer) (Variables|Default|GetSL_LastValue)))
  (Utilities|FlowControl|DoOnce false
    (:Completed
      (bind _returnvalue (Utilities|Time|SetTimerbyEvent (EventDispatchers|CreateEvent :!event Custom|HackyWorkaroundRefreshTimerGuy) 0.1 true))
      (Variables|Default|SetSL_RefreshTimeHandler _returnvalue))))

(event EventOnPostResIn
  (bind _returnvalue (CustomCharacter|GetSkeletalMeshForPartOnPlayerPawn))
  (bind _returnvalue_2 (CustomCharacter|GetSkeletalMeshForPartOnPlayerPawn))
  (|CallParentFunction)
  (CallFunction|ResetSafetyGate)
  (Transformation|AttachComponentToComponent (Variables|Default|GetIdle) _returnvalue "root" "KeepRelative" "KeepRelative" "KeepRelative" true)
  (Transformation|AttachComponentToComponent (Variables|Default|GetUpdate) _returnvalue "root" "KeepRelative" "KeepRelative" "KeepRelative" true)
  (Transformation|AttachComponentToComponent (Variables|Default|GetTop) _returnvalue "root" "KeepRelative" "KeepRelative" "KeepRelative" true)
  ;[1 1 1] SuperLevelling Logic
  (Utilities|FlowControl|Sequence
    (:then_0
      (Utilities|FlowControl|Delay 1.0)
      (Utilities|Time|ClearandInvalidateTimerbyHandle (Variables|Default|GetSL_RefreshTimeHandler)))
    (:then_1
      (bind (_statname _statvalue) (Cosmetics|WatchCosmeticStat (CustomCharacter|GetAssociatedPlayerPawn) "(TagName=\"StatManager.Season.SeasonLevel12\")")
        (:OnStatChanged
          (Rendering|Material|SetScalarParameterValueonMaterials _returnvalue_2 "SL_StartAtLevel" (Math|Conversions|ToFloat(Integer) (Variables|Default|GetSL_StartLevel)))
          (Rendering|Material|SetScalarParameterValueonMaterials _returnvalue_2 "SL_StopAtLevel" (Math|Conversions|ToFloat(Integer) (Variables|Default|GetSL_EndLevel)))
          (Rendering|Material|SetScalarParameterValueonMaterials _returnvalue_2 "SL_EndRadius" (Variables|Default|GetEndRadius))
          (Variables|Default|SetSL_LastValue (Math|Integer|Max(Integer) _statvalue (Variables|Default|GetSL_LastValue)))
          (CallFunction|InitializeVariables)
          (Rendering|Material|SetScalarParameterValueonMaterials (CustomCharacter|GetSkeletalMeshForPartOnPlayerPawn) "SL_CurrentLevel" (Math|Conversions|ToFloat(Integer) (Variables|Default|GetSL_LastValue)))
          (Utilities|FlowControl|DoOnce false
            (:Completed
              (bind _returnvalue_3 (Utilities|Time|SetTimerbyEvent (EventDispatchers|CreateEvent :!event Custom|HackyWorkaroundRefreshTimerGuy) 0.1 true))
              (Variables|Default|SetSL_RefreshTimeHandler _returnvalue_3))))))))

(event Custom|ResetSafetyGate
  ;[1 1 1] SuperLevelling Logic
  (Utilities|FlowControl|DoOnce false
    (:Completed
      (bind _returnvalue (Utilities|Time|SetTimerbyEvent (EventDispatchers|CreateEvent :!event Custom|HackyWorkaroundRefreshTimerGuy) 0.1 true))
      (Variables|Default|SetSL_RefreshTimeHandler _returnvalue))))

(event Custom|Kill>=1
  ;[1 1 1] Watch Cosmetic Stat Athena Kills
  ;;[1 1 1] Update Parameter
  (Effects|Components|ParticleSystem|SetFloatParameter (Variables|Default|GetIdle) "GlowAlpha" 0.6))

(event EventOnCosmeticStateActivated
  (params (Stat%20Name :type Name :default "None") (Stat%20Value :type Integer) (STAGE%20Num%20Being%20Activated :type Integer))
  ;[1 1 1] Watch Cosmetic Stat Athena Kills
  (|CallParentFunction Stat%20Name Stat%20Value STAGE%20Num%20Being%20Activated)
  (bind _output_get (Variables|Default|SetKill%20Stage STAGE%20Num%20Being%20Activated))
  (switch int _output_get
    (:1
      (CallFunction|Kill>=1))
    (:2
      (CallFunction|Kill>=2))
    (:3
      (CallFunction|Kill>=3))
    (:4
      (CallFunction|Kill>=4))
    (:5
      (CallFunction|Kill>=5))))

(event Custom|Kill>=2
  ;[1 1 1] Watch Cosmetic Stat Athena Kills
  ;;[1 1 1] Update Parameter
  (Effects|Components|ParticleSystem|SetFloatParameter (Variables|Default|GetIdle) "GlowAlpha" 0.85))

(event Custom|Kill>=3
  ;[1 1 1] Watch Cosmetic Stat Athena Kills
  ;;[1 1 1] Update Parameter
  (Effects|Components|ParticleSystem|SetFloatParameter (Variables|Default|GetIdle) "GlowAlpha" 1.2))

(event EventOnStatChanged
  (params (State%20Name :type Name :default "None") (Stat%20Value :type Integer))
  ;[1 1 1] Watch Cosmetic Stat Athena Kills
  (|CallParentFunction State%20Name Stat%20Value)
  (if (>= Stat%20Value 5)
    _
    (elif (and (< Stat%20Value 5) (> Stat%20Value (Variables|Default|GetKill%20Stage)))
      _))
  (Components|Activation|Activate (Variables|Default|GetIdle) true))

(event Custom|Kill>=4
  ;[1 1 1] Watch Cosmetic Stat Athena Kills
  ;;[1 1 1] Update Parameter
  (Effects|Components|ParticleSystem|SetFloatParameter (Variables|Default|GetIdle) "GlowAlpha" 1.5))

(event Custom|Kill>=5
  ;[1 1 1] Watch Cosmetic Stat Athena Kills
  ;;[1 1 1] Update Parameter
  (Effects|Components|ParticleSystem|SetFloatParameter (Variables|Default|GetTop) "GlowAlpha" 1.9))

(fn GetAnimInstance
  (return (Components|SkeletalMesh|GetAnimInstance (Customization|GetSkeletalMeshforPartType (CustomCharacter|GetAssociatedPlayerPawn) (CustomCharacter|GetCharacterPartType) :!class /Script/FortniteGame.FortPlayerPawn))))

(fn InitializeVariables
  (bind _returnvalue (CustomCharacter|GetSkeletalMeshForPartOnPlayerPawn))
  (Rendering|Material|SetScalarParameterValueonMaterials _returnvalue "SL_StartAtLevel" (Math|Conversions|ToFloat(Integer) (Variables|Default|GetSL_StartLevel)))
  (Rendering|Material|SetScalarParameterValueonMaterials _returnvalue "SL_StopAtLevel" (Math|Conversions|ToFloat(Integer) (Variables|Default|GetSL_EndLevel)))
  (Rendering|Material|SetScalarParameterValueonMaterials _returnvalue "SL_EndRadius" (Variables|Default|GetEndRadius)))

(fn ConstructionScript
  (|CallParentFunction))
