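; Cosmetic part modifier Blueprint for the Canary character part, adjusting its materials after res-in.
; After calling the parent handler, it checks whether it runs in the front end or on a mobile game, where it only prints debug strings.
; Otherwise it sets the ShadowCrunch scalar parameter to full on the materials of the part's skeletal mesh on the player pawn.
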(event EventOnPostResIn
  (|CallParentFunction)
  (if (CustomCosmeticModifier|IsinFrontEnd)
    (Development|PrintString "Canary FE")
    (if (TouchPlatform|IsMobileGame)
      (Development|PrintString "Canary Is Mobile Game")
      (Rendering|Material|SetScalarParameterValueonMaterials (CustomCharacter|GetSkeletalMeshForPartOnPlayerPawn) "ShadowCrunch" 1.0))))

(fn ConstructionScript)
