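; Character part modifier for the CatBurglar Ghost cosmetic that drives kill-based "SuperLevelling" materials and forces a wrap on weapons and vehicles.
; The highest AthenaKills value feeds SL material parameters, with game-time stamps for effect timing and a short-lived repeating timer as a refresh workaround.
; In the front end, a timeline previews levels 0–6.
; New melee and ranged weapons without excluded tags, plus entered vehicles, get the wrap through the player controller.
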(event Custom|HackyWorkaroundRefreshTimerGuy
  ;[1 1 1] SuperLevelling Logic
  (Utilities|FlowControl|Sequence
    (:then_0
      (CallFunction|InitializeVariabels)
      (Rendering|Material|SetScalarParameterValueonMaterials (CustomCharacter|GetSkeletalMeshForPartOnPlayerPawn) "SL_CurrentLevel" (Math|Conversions|ToFloat(Integer) (stale EventOnPostResIn _output_get)))
      (Utilities|FlowControl|DoOnce false
        (:Completed
          (bind _returnvalue (Utilities|Time|SetTimerbyEvent (delegate Custom|HackyWorkaroundRefreshTimerGuy) 0.1 true))
          (Variables|Default|SetSL_RefreshTimeHandler _returnvalue))))
    (:then_1
      (Utilities|FlowControl|Gate true
        (:Exit
          (if (and (> (Math|Integer|Max(Integer) (stale EventOnPostResIn _statvalue) (Variables|Default|GetSL_LastValue)) (Variables|Default|GetSL_StartLevel)) (< (Math|Integer|Max(Integer) (stale EventOnPostResIn _statvalue) (Variables|Default|GetSL_LastValue)) (+ (Variables|Default|GetSL_EndLevel) 1)))
            (CallFunction|OnActualStatChange)))))))

(event Custom|ResetSafetyGate
  ;[1 1 1] SuperLevelling Logic
  (Utilities|FlowControl|DoOnce false
    (:Completed
      (bind _returnvalue (Utilities|Time|SetTimerbyEvent (delegate Custom|HackyWorkaroundRefreshTimerGuy) 0.1 true))
      (Variables|Default|SetSL_RefreshTimeHandler _returnvalue))))

(event EventOnPostResIn
  (bind _returnvalue (CustomCharacter|GetSkeletalMeshForPartOnPlayerPawn))
  (bind _returnvalue_2 (CustomCharacter|GetSkeletalMeshForPartOnPlayerPawn))
  (bind _sl_lastvalue (Variables|Default|GetSL_LastValue))
  (|CallParentFunction)
  (CallFunction|ResetSafetyGate)
  ;[1 1 1] SuperLevelling Logic
  (Utilities|FlowControl|Sequence
    (:then_0
      (Utilities|FlowControl|Delay 1.0)
      (Utilities|Time|ClearandInvalidateTimerbyHandle (Variables|Default|GetSL_RefreshTimeHandler)))
    (:then_1
      (bind (_statname _statvalue) (Cosmetics|WatchCosmeticStat (CustomCharacter|GetAssociatedPlayerPawn) "(TagName=\"StatManager.AthenaKills\")")
        (:OnStatChanged
          (Rendering|Material|SetScalarParameterValueonMaterials _returnvalue_2 "SL_StartAtLevel" (Math|Conversions|ToFloat(Integer) (Variables|Default|GetSL_StartLevel)))
          (Rendering|Material|SetScalarParameterValueonMaterials _returnvalue_2 "SL_StopAtLevel" (Math|Conversions|ToFloat(Integer) (Variables|Default|GetSL_EndLevel)))
          (Rendering|Material|SetScalarParameterValueonMaterials _returnvalue_2 "SL_EndRadius" (Variables|Default|GetEndRadius))
          (bind _output_get (Variables|Default|SetSL_LastValue (Math|Integer|Max(Integer) _statvalue _sl_lastvalue)))
          (Utilities|FlowControl|Sequence
            (:then_0
              (CallFunction|InitializeVariabels)
              (Rendering|Material|SetScalarParameterValueonMaterials (CustomCharacter|GetSkeletalMeshForPartOnPlayerPawn) "SL_CurrentLevel" (Math|Conversions|ToFloat(Integer) _output_get))
              (Utilities|FlowControl|DoOnce false
                (:Completed
                  (bind _returnvalue_3 (Utilities|Time|SetTimerbyEvent (delegate Custom|HackyWorkaroundRefreshTimerGuy) 0.1 true))
                  (Variables|Default|SetSL_RefreshTimeHandler _returnvalue_3))))
            (:then_1
              (Utilities|FlowControl|Gate true
                (:Exit
                  (if (and (> (Math|Integer|Max(Integer) _statvalue _sl_lastvalue) (Variables|Default|GetSL_StartLevel)) (< (Math|Integer|Max(Integer) _statvalue _sl_lastvalue) (+ (Variables|Default|GetSL_EndLevel) 1)))
                    (CallFunction|OnActualStatChange)))))))))
    (:then_2
      (if (CustomCosmeticModifier|IsinFrontEnd)
        (Timeline|Timeline_0 PlayFromStart
          (:Stage%201
            (Rendering|Material|SetScalarParameterValueonMaterials _returnvalue "SL_CurrentLevel" 0.0))
          (:Stage%202
            (Rendering|Material|SetScalarParameterValueonMaterials _returnvalue "SL_CurrentLevel" 1.0))
          (:Stage%203
            (Rendering|Material|SetScalarParameterValueonMaterials _returnvalue "SL_CurrentLevel" 2.0))
          (:Stage%204
            (Rendering|Material|SetScalarParameterValueonMaterials _returnvalue "SL_CurrentLevel" 3.0))
          (:Stage%205
            (Rendering|Material|SetScalarParameterValueonMaterials _returnvalue "SL_CurrentLevel" 4.0))
          (:Stage%206
            (Rendering|Material|SetScalarParameterValueonMaterials _returnvalue "SL_CurrentLevel" 5.0))
          (:Step%207
            (Rendering|Material|SetScalarParameterValueonMaterials _returnvalue "SL_CurrentLevel" 6.0)))
        (Rendering|Material|SetScalarParameterValueonMaterials (CustomCharacter|GetSkeletalMeshForPartOnPlayerPawn) "SL_GameTimeFromBP" (Utilities|Time|GetGameTimeinSeconds))))))

(event Custom|OnActualStatChange
  ;[1 1 1] SuperLevelling Logic
  ;;[1 1 1] This should only fire off when the stat changes - use this event!
  (Rendering|Material|SetScalarParameterValueonMaterials (CustomCharacter|GetSkeletalMeshForPartOnPlayerPawn) "SL_GameTimeFromBP" (Utilities|Time|GetGameTimeinSeconds)))

(event EventBeginPlay)

(event Custom|InitializeVariabels
  (bind _returnvalue (CustomCharacter|GetSkeletalMeshForPartOnPlayerPawn))
  (Rendering|Material|SetScalarParameterValueonMaterials _returnvalue "SL_StartAtLevel" (Math|Conversions|ToFloat(Integer) (Variables|Default|GetSL_StartLevel)))
  (Rendering|Material|SetScalarParameterValueonMaterials _returnvalue "SL_StopAtLevel" (Math|Conversions|ToFloat(Integer) (Variables|Default|GetSL_EndLevel)))
  (Rendering|Material|SetScalarParameterValueonMaterials _returnvalue "SL_EndRadius" (Variables|Default|GetEndRadius)))

(event EventOnNewWeapons
  (params (WeaponsAsItems :type Array%20of%20Fort%20Item%20Object%20References))
  (bind _pendingitemstowrap (Variables|Default|GetPendingItemsToWrap))
  ;[1 1 1] Force Weapons And Vehicles To Have A Specific Wrap (MIDAS TOUCH)
  (Utilities|NotwhenBID)
  (for _array_element WeaponsAsItems
    ; Prevent Consumables and Sprites from having wraps
    (if (not (GameplayTags|HasAnyTags (GameplayTags|GetOwnedGameplayTags (Item|GetItemDefinition _array_element) :!class /Script/GameplayTags.BlueprintGameplayTagLibrary) (Variables|Default|GetGameplay%20Tags%20To%20Find)))
      (switch Utilities|FlowControl|Switch|SwitchonEFortItemType (Item|GetType _array_element)
        (:WeaponMelee
          _)
        (:WeaponRanged
          _))
      (Utilities|Array|Add (Variables|Default|GetPendingItemsToWrap) _array_element)))
  (if (Utilities|IsValidSoftObjectReference (Variables|Default|GetWeaponWrapToUse))
    (CallFunction|ApplyOverridetoItems _pendingitemstowrap)
    (Utilities|Array|Clear _pendingitemstowrap)))

(event EventOnEnterVehicle
  (params (VehicleEntered :type Fort%20Vehicle%20Interface%20Interface%20%28by%20ref%29))
  ;[1 1 1] Force Weapons And Vehicles To Have A Specific Wrap (MIDAS TOUCH)
  (Utilities|NotwhenBID)
  (if (Utilities|IsValidSoftObjectReference (Variables|Default|GetWeaponWrapToUse))
    (bind _asfort_player_controller (Default|GetAssociatedFortPlayerController))
    (Utilities|IsValid _asfort_player_controller
      (:Is%20Valid
        (Class|FortPlayerController|ApplyOverrideWraptoVehicle _asfort_player_controller (Variables|Default|GetWeaponWrapToUse))))))

(fn ApplyOverrideToItems
  (params (Array :type Array%20of%20Fort%20Item%20Object%20References))
  (bind _asfort_player_controller (Default|GetAssociatedFortPlayerController))
  (Utilities|IsValid _asfort_player_controller
    (:Is%20Valid
      (for _array_element Array
        (Class|FortPlayerController|ServerApplyOverrideWraptoItem _asfort_player_controller (Item|GetItemGuid _array_element) (Variables|Default|GetWeaponWrapToUse))))))

(fn ConstructionScript
  (bind _returnvalue (CustomCharacter|GetSkeletalMeshForPartOnPlayerPawn))
  (|CallParentFunction)
  (Rendering|Material|SetScalarParameterValueonMaterials (CustomCharacter|GetSkeletalMeshForPartOnPlayerPawn) "SL_GameTimeFromBP" -99999.0)
  (Rendering|Material|SetScalarParameterValueonMaterials _returnvalue "SL_StartAtLevel" (Math|Conversions|ToFloat(Integer) (Variables|Default|GetSL_StartLevel)))
  (Rendering|Material|SetScalarParameterValueonMaterials _returnvalue "SL_StopAtLevel" (Math|Conversions|ToFloat(Integer) (Variables|Default|GetSL_EndLevel))))
