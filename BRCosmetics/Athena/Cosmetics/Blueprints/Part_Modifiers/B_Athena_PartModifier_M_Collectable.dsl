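; Character part modifier for the Athena M_Collectable cosmetic that drives reactive facial expressions and eye blinking through scalar material parameters.
; On post-res-in, it watches stats for damage dealt or received, health or shield gains and opened chests, setting a timestamped expression and briefly pausing blinks.
; A randomized timer drives blinks, with a chance of doubles.
; DBNO switches the neutral expression to sad and suppresses blinking until it ends.
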(event Custom|Blink_Eyes
  (bind _returnvalue (CustomCharacter|GetSkeletalMeshForPartOnPlayerPawn))
  ;[1 1 1] Blink
  (if (not (Variables|Default|GetIn%20Is%20DBNO))
    (Utilities|FlowControl|Delay 0.075)
    (Rendering|Material|SetScalarParameterValueonMaterials _returnvalue "Collectable_FE_Blink_Amount" (Variables|Default|GetBlinkAmount))
    (Utilities|FlowControl|Delay 0.12)
    (Rendering|Material|SetScalarParameterValueonMaterials _returnvalue "Collectable_FE_Blink_Amount" 0.0)
    (if (and (< (Math|Random|RandomFloat) (Variables|Default|GetDouble_Blink_Chance)) (not (Variables|Default|GetHas_Double_Blinked)))
      (Variables|Default|SetHas_Double_Blinked true)
      (CallFunction|BlinkEyes)
      (else
        (Variables|Default|SetHas_Double_Blinked false)
        (CallFunction|NewBlinkTimer)))))

(event Custom|New_Blink_Timer
  ;[1 1 1] Blink Timer
  (if (Utilities|Time|IsValidTimerHandle (Variables|Default|GetBlink_Timer_Handle))
    (Utilities|Time|ClearandInvalidateTimerbyHandle (Variables|Default|GetBlink_Timer_Handle))
    (else
      _))
  (bind _returnvalue (Utilities|Time|SetTimerbyFunctionName 0 "Blink_Eyes" (Math|Random|RandomFloatinRange 2.0 5.0)))
  (Variables|Default|SetBlink_Timer_Handle _returnvalue))

(event Custom|Pause_Blinking
  (bind _blink_timer_handle (Variables|Default|GetBlink_Timer_Handle))
  (bind _blink_timer_handle_2 (Variables|Default|GetBlink_Timer_Handle))
  ;[1 1 1] Pause Blinking
  (Variables|Default|SetHas_Double_Blinked true) ; Needed to avoid a timing edge case
  (Rendering|Material|SetScalarParameterValueonMaterials (CustomCharacter|GetSkeletalMeshForPartOnPlayerPawn) "Collectable_FE_Blink_Amount" 0.0)
  (if (Utilities|Time|IsValidTimerHandle _blink_timer_handle_2)
    (Utilities|Time|PauseTimerbyHandle _blink_timer_handle_2)
    (Utilities|FlowControl|RetriggerableDelay 3.0)
    (if (Utilities|Time|IsValidTimerHandle _blink_timer_handle)
      (Utilities|Time|UnpauseTimerbyHandle _blink_timer_handle)
      (else
        (CallFunction|NewBlinkTimer)))))

(event EventOnPostResIn
  (bind _returnvalue (CustomCharacter|GetSkeletalMeshForPartOnPlayerPawn))
  (bind _returnvalue_2 (CustomCharacter|GetSkeletalMeshForPartOnPlayerPawn))
  (bind _returnvalue_3 (CustomCharacter|GetSkeletalMeshForPartOnPlayerPawn))
  (bind _returnvalue_4 (CustomCharacter|GetSkeletalMeshForPartOnPlayerPawn))
  (bind _returnvalue_5 (CustomCharacter|GetSkeletalMeshForPartOnPlayerPawn))
  (|CallParentFunction)
  (Utilities|FlowControl|Sequence
    (:then_0
      ;[1 1 1] Blink Timer
      (if (Utilities|Time|IsValidTimerHandle (Variables|Default|GetBlink_Timer_Handle))
        (Utilities|Time|ClearandInvalidateTimerbyHandle (Variables|Default|GetBlink_Timer_Handle))
        (else
          _))
      (bind _returnvalue_6 (Utilities|Time|SetTimerbyFunctionName 0 "Blink_Eyes" (Math|Random|RandomFloatinRange 2.0 5.0)))
      (Variables|Default|SetBlink_Timer_Handle _returnvalue_6))
    (:then_1
      ;[1 1 1] DamageDealt
      (Cosmetics|WatchCosmeticStat (CustomCharacter|GetAssociatedPlayerPawn) "(TagName=\"StatManager.DamageDealt\")"
        (:OnStatChanged
          (Rendering|Material|SetScalarParameterValueonMaterials _returnvalue_3 "Collection_GameTimeFromBP" (Utilities|Time|GetGameTimeinSeconds))
          (Rendering|Material|SetScalarParameterValueonMaterials _returnvalue_3 "Collection_ReactiveLength" 1.5)
          (Rendering|Material|SetScalarParameterValueonMaterials _returnvalue_3 "Collectable_Expression" 0.7)
          (CallFunction|PauseBlinking))))
    (:then_2
      ;[1 1 1] Damage Received
      (Cosmetics|WatchCosmeticStat (CustomCharacter|GetAssociatedPlayerPawn) "(TagName=\"StatManager.DamageReceived\")"
        (:OnStatChanged
          (Rendering|Material|SetScalarParameterValueonMaterials _returnvalue_5 "Collection_GameTimeFromBP" (Utilities|Time|GetGameTimeinSeconds))
          (Rendering|Material|SetScalarParameterValueonMaterials _returnvalue_5 "Collection_ReactiveLength" 1.5)
          (Rendering|Material|SetScalarParameterValueonMaterials _returnvalue_5 "Collectable_Expression" 0.86)
          (CallFunction|PauseBlinking))))
    (:then_3
      ;[1 1 1] Health
      (Cosmetics|WatchCosmeticStat (CustomCharacter|GetAssociatedPlayerPawn) "(TagName=\"StatManager.GainedHealthTimes\")"
        (:OnStatChanged
          (Rendering|Material|SetScalarParameterValueonMaterials _returnvalue_2 "Collection_GameTimeFromBP" (Utilities|Time|GetGameTimeinSeconds))
          (Rendering|Material|SetScalarParameterValueonMaterials _returnvalue_2 "Collection_ReactiveLength" 2.5)
          (Rendering|Material|SetScalarParameterValueonMaterials _returnvalue_2 "Collectable_Expression" 0.5)
          (CallFunction|PauseBlinking))))
    (:then_4
      ;[1 1 1] Shield
      (Cosmetics|WatchCosmeticStat (CustomCharacter|GetAssociatedPlayerPawn) "(TagName=\"StatManager.GainedShieldTimes\")"
        (:OnStatChanged
          (Rendering|Material|SetScalarParameterValueonMaterials _returnvalue_4 "Collection_GameTimeFromBP" (Utilities|Time|GetGameTimeinSeconds))
          (Rendering|Material|SetScalarParameterValueonMaterials _returnvalue_4 "Collection_ReactiveLength" 2.5)
          (Rendering|Material|SetScalarParameterValueonMaterials _returnvalue_4 "Collectable_Expression" 0.5)
          (CallFunction|PauseBlinking))))
    (:then_5
      ;[1 1 1] openChest
      (Cosmetics|WatchCosmeticStat (CustomCharacter|GetAssociatedPlayerPawn) "(TagName=\"StatManager.OpenChests\")"
        (:OnStatChanged
          (Rendering|Material|SetScalarParameterValueonMaterials _returnvalue "Collection_GameTimeFromBP" (Utilities|Time|GetGameTimeinSeconds))
          (Rendering|Material|SetScalarParameterValueonMaterials _returnvalue "Collection_ReactiveLength" 2.5)
          (Rendering|Material|SetScalarParameterValueonMaterials _returnvalue "Collectable_Expression" 0.76)
          (CallFunction|PauseBlinking))))
    (:then_6
      ;[1 1 1] Clear initial expression
      (Rendering|Material|SetScalarParameterValueonMaterials (CustomCharacter|GetSkeletalMeshForPartOnPlayerPawn) "Collection_GameTimeFromBP" -1e+01))
    (:then_7
      (EventDispatchers|BindEventtoOnDBNOStateChanged (CustomCharacter|GetAssociatedPlayerPawn) (delegate Custom|DBNO_Changed)))))

(event Custom|DBNO_Changed
  (params (FortPawn :type Fort%20Pawn%20Object%20Reference) (bInIsDBNO :type Boolean))
  (bind _returnvalue (CustomCharacter|GetSkeletalMeshForPartOnPlayerPawn))
  (Rendering|Material|SetScalarParameterValueonMaterials (CustomCharacter|GetSkeletalMeshForPartOnPlayerPawn) "Collection_GameTimeFromBP" -1e+01)
  (bind _output_get (Variables|Default|SetIn%20Is%20DBNO bInIsDBNO))
  ;[1 1 1] Set default expression to Sad when DBNO
  (if _output_get
    (Rendering|Material|SetScalarParameterValueonMaterials _returnvalue "Collectable_Neutral_Expresssion" 0.86)
    (else
      (Rendering|Material|SetScalarParameterValueonMaterials _returnvalue "Collectable_Neutral_Expresssion" 0.56)
      (CallFunction|NewBlinkTimer))))

(fn ConstructionScript)
