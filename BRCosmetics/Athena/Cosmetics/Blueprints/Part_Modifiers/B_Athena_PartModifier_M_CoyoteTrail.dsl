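; Part modifier for the CoyoteTrail reactive cosmetic, which transforms when shield fills, unless tagged NonReactive.
; In game it watches the local pawn's shield, or the shield attribute for Save The World, to request a preset cosmetic swap.
; After res-in, the Activated tag or earlier-replicating reactive part fires idle VFX and timestamped material transition once.
; The front end previews the transition forward and back on timed delays, undoing the swap afterward.
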(event EventOnPawnAndPartChanged
  (bind _returnvalue (CallFunction|GetIdleVFXComponent))
  (bind _returnvalue_2 (CallFunction|GetIdleVFXComponent))
  (bind _returnvalue_3 (CustomCharacter|GetAssociatedPlayerPawn))
  (|CallParentFunction)
  (if (GameplayTags|HasTag (Class|FortCosmeticFunctionLibrary|%28Deprecated%29GetLoadoutMetaTagsForItem (Variables|/Script/FortniteGame.FortPlayerPawn|GetCosmeticLoadout (CustomCharacter|GetAssociatedPlayerPawn)) (Variables|Default|GetItemToModify)) "(TagName=\"Cosmetics.MetaTags.CoyoteTrail.NonReactive\")")
    _
    (elif (CustomCosmeticModifier|IsinFrontEnd)
      ;[1 1 1] Front End Loop
      ;;[0 0.2627 1] Skip the initial delay for the face part since it is swapped after the animation
      (if (Utilities|Enum|Equal(Enum) (CustomCharacter|GetCharacterPartType) (Variables|Default|GetSkip%20Ahead%20On%20Part%20Type))
        (Components|Activation|SetActive _returnvalue_2 true true)
        (Effects|Components|ParticleSystem|SetBoolParameter _returnvalue_2 "Return" false)
        (CallFunction|SetMaterialParameters 1.0)
        (CallFunction|PushMaterialParameters)
        (CallFunction|SwapVariant)
        (Utilities|FlowControl|Delay (Variables|Default|GetFully%20Activated%20TIme))
        (Components|Activation|SetActive _returnvalue true true)
        (Effects|Components|ParticleSystem|SetBoolParameter _returnvalue "Return" true)
        (CallFunction|SetMaterialParameters 1.0 1.0)
        (CallFunction|PushMaterialParameters)
        (Utilities|FlowControl|Delay (Variables|Default|GetAnimation%20Duration))
        (CallFunction|UndoSwapFrontendOnly)
        (Utilities|FlowControl|Delay (Variables|Default|GetRestart%20Delay)))
      (Utilities|FlowControl|Delay (Variables|Default|GetFrontend%20Start%20Delay))
      (else
        (Utilities|FlowControl|Sequence
          (:then_0
            ;[1 1 1] Watch health on a local pawn to trigger an RPC
            (if (Pawn|IsLocallyControlled _returnvalue_3)
              (bind (_health _maxhealth _shield _maxshield) (Cosmetics|WatchPawnHealth _returnvalue_3)
                (:then
                  (Utilities|FlowControl|Delay 0.1))
                (:OnHitPointsChanged
                  ;[1 1 1] Check that our shield is at its max value, and that the max is not 0
                  (if (and (not (Math|Float|NearlyEqual(Float) _maxshield 0.0 (Variables|Default|GetError%20Tolerance))) (Math|Float|NearlyEqual(Float) _shield _maxshield (Variables|Default|GetError%20Tolerance)))
                    (Utilities|FlowControl|Delay 0.2))))
              (Utilities|FlowControl|Gate true
                (:Exit
                  (Utilities|FlowControl|Sequence
                    (:then_0
                      _)
                    (:then_1
                      _))
                  (CallFunction|SwapVariant)))))
          (:then_1
            ;[1 1 1] Trigger for Save The World
            (bind (_asynctaskproxy _attribute _newvalue) (Ability|Tasks|WaitForAttributeChanged _returnvalue_3 (Variables|Default|GetAttShield))
              (:Changed
                (if (>= _newvalue 1.5e+02)
                  (Utilities|FlowControl|Delay 0.2))))))))))

(event EventOnPostResIn
  (|CallParentFunction)
  (Utilities|FlowControl|Sequence
    (:then_0
      (if (not (or (CustomCosmeticModifier|IsinFrontEnd) (Variables|Default|GetTriggeredTransition)))
        (Utilities|FlowControl|Sequence
          (:then_0
            ;[1 1 1] Check if the part or meta tag is present - whichever happens first
            (if (GameplayTags|HasTag (Class|FortCosmeticFunctionLibrary|%28Deprecated%29GetLoadoutMetaTagsForItem (Variables|/Script/FortniteGame.FortPlayerPawn|GetCosmeticLoadout (CustomCharacter|GetAssociatedPlayerPawn)) (Variables|Default|GetItemToModify)) "(TagName=\"Cosmetics.MetaTags.CoyoteTrail.Activated\")")
              ;[1 1 1] Save the transition time and trigger the FX on the first activation only
              (Variables|Default|SetTriggeredTransition true)
              (CallFunction|SetMaterialParameters 1.0)
              (Components|Activation|SetActive (CallFunction|GetIdleVFXComponent) true true)))
          (:then_1
            ;[1 1 1] A bit gross, but we check to see if the character part has changed since that will replicate before the loadout has changed
            (bind _delegatehandlecontroller_out (Cosmetics|WhenCosmeticPartApplied (CustomCharacter|GetAssociatedPlayerPawn) (delegate Custom|CheckPart) true))
            (Variables|Default|SetDelegateHandleController _delegatehandlecontroller_out)))))
    (:then_1
      (CallFunction|PushMaterialParameters))))

(event Custom|CheckPart
  (params (Pawn :type Fort%20Player%20Pawn%20Object%20Reference) (Part :type Custom%20Character%20Part%20Object%20Reference) (MeshPart :type Skeletal%20Mesh%20Component%20Object%20Reference) (PartType :type EFortCustomPartType%20Enum :default "Head"))
  ;[1 1 1] A bit gross, but we check to see if the character part has changed since that will replicate before the loadout has changed
  (Utilities|IsValid Part
    (:Is%20Valid
      (if (and (== Part (Utilities|ResolveSoftReference (Variables|Default|GetReactivePart))) (not (Variables|Default|GetTriggeredTransition)))
        ;[1 1 1] Save the transition time and trigger the FX on the first activation only
        (Variables|Default|SetTriggeredTransition true)
        (CallFunction|SetMaterialParameters 1.0)
        (Components|Activation|SetActive (CallFunction|GetIdleVFXComponent) true true)))))

(fn checkIfBlackAlt
  (bind _index (CosmeticLoadout|GetVariantChannelActiveIndexforSlotActor "(TagName=\"LoadoutSlot.Character\")" (CustomCharacter|GetAssociatedPlayerPawn) "(TagName=\"Cosmetics.Variant.Channel.Material\")" (Utilities|Array|MakeArray "(TagName=\"Cosmetics.Variant.Property.Mat1\")" "(TagName=\"Cosmetics.Variant.Property.Mat2\")" "(TagName=\"Cosmetics.Variant.Property.Mat3\")" "(TagName=\"Cosmetics.Variant.Property.Mat4\")" "(TagName=\"Cosmetics.Variant.Property.Mat5\")" "(TagName=\"Cosmetics.Variant.Property.Mat6\")" "(TagName=\"Cosmetics.Variant.Property.Mat7\")")))
  (switch int _index
    (:Default
      _)
    (:0
      (Variables|Default|SetisBlackAlt true))
    (:1
      _)
    (:2
      _)
    (:3
      _)
    (:4
      _)
    (:5
      _)
    (:6
      _))
  (Variables|Default|SetisBlackAlt false))

(fn checkIfPinkAlt
  (bind _index (CosmeticLoadout|GetVariantChannelActiveIndexforSlotActor "(TagName=\"LoadoutSlot.Character\")" (CustomCharacter|GetAssociatedPlayerPawn) "(TagName=\"Cosmetics.Variant.Channel.Material\")" (Utilities|Array|MakeArray "(TagName=\"Cosmetics.Variant.Property.Mat1\")" "(TagName=\"Cosmetics.Variant.Property.Mat2\")" "(TagName=\"Cosmetics.Variant.Property.Mat3\")" "(TagName=\"Cosmetics.Variant.Property.Mat4\")" "(TagName=\"Cosmetics.Variant.Property.Mat5\")" "(TagName=\"Cosmetics.Variant.Property.Mat6\")" "(TagName=\"Cosmetics.Variant.Property.Mat7\")")))
  (switch int _index
    (:0
      _)
    (:1
      (Variables|Default|SetisPinkAlt true))
    (:2
      _)
    (:3
      _)
    (:4
      _)
    (:5
      _)
    (:6
      _))
  (Variables|Default|SetisPinkAlt false))

(fn PushMaterialParameters
  (bind _returnvalue (CustomCharacter|GetSkeletalMeshForPartOnPlayerPawn))
  (Rendering|Material|SetScalarParameterValueonMaterials _returnvalue "TransitionTimeFromBP" (Variables|MaterialParameters|GetTransitionTime))
  (Rendering|Material|SetScalarParameterValueonMaterials _returnvalue "Level_Progress" (Variables|MaterialParameters|GetLevelProgress))
  (Rendering|Material|SetScalarParameterValueonMaterials _returnvalue "TransitionBackwards" (Variables|MaterialParameters|GetTransitionBackwards)))

(fn SetMaterialParameters
  (params (NewLevelProgress :type Float%20%28double-precision%29) (NewTransitionBackwards :type Float%20%28double-precision%29))
  (Variables|MaterialParameters|SetLevelProgress NewLevelProgress)
  (Variables|MaterialParameters|SetTransitionBackwards NewTransitionBackwards)
  (Variables|MaterialParameters|SetTransitionTime (Utilities|Time|GetGameTimeinSeconds)))

(fn SwapVariant
  (if (Utilities|Enum|Equal(Enum) (CustomCharacter|GetCharacterPartType) (Variables|Default|GetSwap%20Only%20On%20Part%20Type))
    (Class|FortCosmeticFunctionLibrary|RequestPresetCosmeticSwap (CustomCharacter|GetAssociatedPlayerPawn) (Variables|Default|GetItemToModify))))

(fn UndoSwapFrontendOnly
  (if (Utilities|Enum|Equal(Enum) (CustomCharacter|GetCharacterPartType) "Body")
    (Class|FortPlayerPawn|DisableCosmeticSwap (CustomCharacter|GetAssociatedPlayerPawn) (Variables|Default|GetSwap%20Id))))

(fn ConstructionScript)
