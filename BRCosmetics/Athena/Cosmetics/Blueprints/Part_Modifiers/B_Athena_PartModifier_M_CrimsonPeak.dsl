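; Cosmetic part modifier for the CrimsonPeak character that drives a damage-reactive material and idle VFX.
; After res-in it resets the idle effect's "Health" parameter, then continues only if the Material variant channel resolves to index 0 (Mat0).
; In the front end it plays a preview timeline, while elsewhere it watches the DamageReceived stat and plays Crimson_Damage on each change.
; Both timelines drive the "IsDamagedFromBP" material scalar and the VFX "Health" parameter.
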(event EventOnPostResIn
  (bind _returnvalue (CustomCharacter|GetAssociatedPlayerPawn))
  (|CallParentFunction)
  (Effects|Components|ParticleSystem|SetFloatParameter (CallFunction|GetIdleVFXComponent) "Health" 1.0)
  (bind _asplayer_pawn_athena (Utilities|Casting|CastToPlayerPawn_Athena _returnvalue))
  (bind _index (CosmeticLoadout|GetVariantChannelActiveIndexforSlotActor "(TagName=\"LoadoutSlot.Character\")" _asplayer_pawn_athena "(TagName=\"Cosmetics.Variant.Channel.Material\")" (Utilities|Array|MakeArray "(TagName=\"Cosmetics.Variant.Property.Mat0\")" "(TagName=\"Cosmetics.Variant.Property.Mat1\")") false false))
  (switch int _index
    (:0
      (if (CustomCosmeticModifier|IsinFrontEnd)
        (bind (_direction _damage _healthfx) (Timeline|FE PlayFromStart)
          (:Update
            (Rendering|Material|SetScalarParameterValueonMaterials (CustomCharacter|GetSkeletalMeshForPartOnPlayerPawn) "IsDamagedFromBP" _damage)
            (Effects|Components|ParticleSystem|SetFloatParameter (CallFunction|GetIdleVFXComponent) "Health" _healthfx)))
        (else
          (Cosmetics|WatchCosmeticStat _returnvalue "(TagName=\"StatManager.DamageReceived\")" 0 false
            (:OnStatChanged
              (bind (_direction_2 _damage_2 _healthfx_2) (Timeline|Crimson_Damage PlayFromStart)
                (:Update
                  (Rendering|Material|SetScalarParameterValueonMaterials (CustomCharacter|GetSkeletalMeshForPartOnPlayerPawn) "IsDamagedFromBP" _damage_2)
                  (Effects|Components|ParticleSystem|SetFloatParameter (CallFunction|GetIdleVFXComponent) "Health" _healthfx_2))))))))))

(fn ConstructionScript)
