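; Cosmetic part modifier for the Axl backpack (M_MED_Axl_Pack) that lights up in response to the player's kills.
; After res-in, it watches the AthenaKills stat on the associated player pawn and stores each new value as the kill count.
; When the count is above zero, it activates the idle VFX and restarts a timeline that drives the LightIntensity material parameter on the part's skeletal mesh.
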(event EventOnPostResIn
  (|CallParentFunction)
  (bind (_statname _statvalue) (Cosmetics|WatchCosmeticStat (CustomCharacter|GetAssociatedPlayerPawn) "(TagName=\"StatManager.AthenaKills\")")
    (:OnStatChanged
      (bind _output_get (Variables|Default|SetKillCount (Math|Conversions|ToFloat(Integer) _statvalue)))
      (if (> _output_get 0.0)
        (Components|Activation|Activate (CallFunction|GetIdleVFXComponent) false)
        (bind (_direction _slurpemissivemultiplier) (Timeline|Timeline_0 PlayFromStart)
          (:Update
            (Rendering|Material|SetScalarParameterValueonMaterials (CustomCharacter|GetSkeletalMeshForPartOnPlayerPawn) "LightIntensity" _slurpemissivemultiplier)))))))

(fn ConstructionScript)
