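; Cosmetic part modifier for the Neon character part that makes its materials react to the wearer's gameplay events.
; After res-in, it watches cosmetic stats for damage dealt or received, health and shield gains, eliminations and chest opens.
; Each change stamps game time into the materials with an event-specific color, length and shape, so no Blueprint tick is needed.
; Spawn backdates that timestamp to suppress a reaction, and the front end adjusts audio-reactive and wobble parameters.
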(event EventOnPostResIn
  (bind _returnvalue (CustomCharacter|GetSkeletalMeshForPartOnPlayerPawn))
  (bind _returnvalue_2 (CustomCharacter|GetSkeletalMeshForPartOnPlayerPawn))
  (bind _returnvalue_3 (CustomCharacter|GetSkeletalMeshForPartOnPlayerPawn))
  (bind _returnvalue_4 (CustomCharacter|GetSkeletalMeshForPartOnPlayerPawn))
  (bind _returnvalue_5 (CustomCharacter|GetSkeletalMeshForPartOnPlayerPawn))
  (bind _returnvalue_6 (CustomCharacter|GetSkeletalMeshForPartOnPlayerPawn))
  (bind _returnvalue_7 (CustomCharacter|GetSkeletalMeshForPartOnPlayerPawn))
  (|CallParentFunction)
  (Utilities|FlowControl|Sequence
    (:then_0
      ;[1 1 1] DamageDealt
      (Cosmetics|WatchCosmeticStat (CustomCharacter|GetAssociatedPlayerPawn) "(TagName=\"StatManager.DamageDealt\")"
        (:OnStatChanged
          (Rendering|Material|SetScalarParameterValueonMaterials _returnvalue_5 "Neon_GameTimeFromBP" (Utilities|Time|GetGameTimeinSeconds))
          (Rendering|Material|SetScalarParameterValueonMaterials _returnvalue_5 "Neon_ReactiveLength" 0.5)
          (Rendering|Material|SetVectorParameterValueonMaterials _returnvalue_5 "Neon_ReactiveColor" "1.000000,1.000000,1.000000")
          (Rendering|Material|SetScalarParameterValueonMaterials _returnvalue_5 "Neon_ReactiveShape" 0.75))))
    (:then_1
      ;[1 1 1] Damage Received
      (Cosmetics|WatchCosmeticStat (CustomCharacter|GetAssociatedPlayerPawn) "(TagName=\"StatManager.DamageReceived\")"
        (:OnStatChanged
          (Rendering|Material|SetScalarParameterValueonMaterials _returnvalue_3 "Neon_GameTimeFromBP" (Utilities|Time|GetGameTimeinSeconds))
          (Rendering|Material|SetScalarParameterValueonMaterials _returnvalue_3 "Neon_ReactiveLength" 0.8)
          (Rendering|Material|SetVectorParameterValueonMaterials _returnvalue_3 "Neon_ReactiveColor" "1.000000,0.000000,0.000000")
          (Rendering|Material|SetScalarParameterValueonMaterials _returnvalue_3 "Neon_ReactiveShape" 0.5))))
    (:then_2
      ;[1 1 1] Health
      (Cosmetics|WatchCosmeticStat (CustomCharacter|GetAssociatedPlayerPawn) "(TagName=\"StatManager.GainedHealthTimes\")"
        (:OnStatChanged
          (Rendering|Material|SetScalarParameterValueonMaterials _returnvalue_4 "Neon_GameTimeFromBP" (Utilities|Time|GetGameTimeinSeconds))
          (Rendering|Material|SetScalarParameterValueonMaterials _returnvalue_4 "Neon_ReactiveLength" 3.0)
          (Rendering|Material|SetVectorParameterValueonMaterials _returnvalue_4 "Neon_ReactiveColor" "0.000000,1.000000,0.000000")
          (Rendering|Material|SetScalarParameterValueonMaterials _returnvalue_4 "Neon_ReactiveShape" 1.0))))
    (:then_3
      ;[1 1 1] Shield
      (Cosmetics|WatchCosmeticStat (CustomCharacter|GetAssociatedPlayerPawn) "(TagName=\"StatManager.GainedShieldTimes\")"
        (:OnStatChanged
          (Rendering|Material|SetScalarParameterValueonMaterials _returnvalue_7 "Neon_GameTimeFromBP" (Utilities|Time|GetGameTimeinSeconds))
          (Rendering|Material|SetScalarParameterValueonMaterials _returnvalue_7 "Neon_ReactiveLength" 3.0)
          (Rendering|Material|SetVectorParameterValueonMaterials _returnvalue_7 "Neon_ReactiveColor" "0.000000,0.000000,1.000000")
          (Rendering|Material|SetScalarParameterValueonMaterials _returnvalue_7 "Neon_ReactiveShape" 1.0))))
    (:then_4
      ;[1 1 1] Elimination
      (Cosmetics|WatchCosmeticStat (CustomCharacter|GetAssociatedPlayerPawn) "(TagName=\"StatManager.AthenaKills\")"
        (:OnStatChanged
          (Rendering|Material|SetScalarParameterValueonMaterials _returnvalue_6 "Neon_GameTimeFromBP" (Utilities|Time|GetGameTimeinSeconds))
          (Rendering|Material|SetScalarParameterValueonMaterials _returnvalue_6 "Neon_ReactiveLength" 3.0)
          (Rendering|Material|SetVectorParameterValueonMaterials _returnvalue_6 "Neon_ReactiveColor" "3.000000,2.000000,1.000000")
          (Rendering|Material|SetScalarParameterValueonMaterials _returnvalue_6 "Neon_ReactiveShape" 0.75))))
    (:then_5
      ;[1 1 1] openChest
      (Cosmetics|WatchCosmeticStat (CustomCharacter|GetAssociatedPlayerPawn) "(TagName=\"StatManager.OpenChests\")"
        (:OnStatChanged
          (Rendering|Material|SetScalarParameterValueonMaterials _returnvalue_2 "Neon_GameTimeFromBP" (Utilities|Time|GetGameTimeinSeconds))
          (Rendering|Material|SetScalarParameterValueonMaterials _returnvalue_2 "Neon_ReactiveLength" 3.0)
          (Rendering|Material|SetVectorParameterValueonMaterials _returnvalue_2 "Neon_ReactiveColor" "1.000000,0.900000,0.000000")
          (Rendering|Material|SetScalarParameterValueonMaterials _returnvalue_2 "Neon_ReactiveShape" 0.75))))
    (:then_6
      ;[1 1 1] Stop Reaction on Spawn + Reduce Audio Amplitude in FE
      (Rendering|Material|SetScalarParameterValueonMaterials _returnvalue "Neon_GameTimeFromBP" -1e+01)
      (if (CustomCosmeticModifier|IsinFrontEnd)
        (Rendering|Material|SetScalarParameterValueonMaterials _returnvalue "Neon_EventAudioMultiplier" 0.3)
        (Rendering|Material|SetScalarParameterValueonMaterials _returnvalue "Neon_AudioColorPulseStrength" 0.1)
        (Rendering|Material|SetVectorParameterValueonMaterials _returnvalue "Neon_WobbleSpeed" "0.140000,-0.350000, 0")))))

(fn GameTimeFromBPForGlow
  (params (A :type Float%20%28double-precision%29) (B :type Float%20%28double-precision%29))
  (return 0.0))

(fn ConstructionScript)
