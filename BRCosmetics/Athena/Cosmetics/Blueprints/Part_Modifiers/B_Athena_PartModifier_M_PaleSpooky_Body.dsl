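; Cosmetic part modifier Blueprint for the male Pale Spooky body in Athena, controlling its idle VFX during skydiving.
; When skydiving begins, it calls the parent handler, then deactivates the idle VFX component.
; When skydiving ends, it calls the parent handler, then reactivates that component with a reset.
; The construction script is empty.
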(event EventOnBeginSkydiving
  (|CallParentFunction)
  (Components|Activation|Deactivate (CallFunction|GetIdleVFXComponent)))

(event EventOnEndSkydiving
  (|CallParentFunction)
  (Components|Activation|Activate (CallFunction|GetIdleVFXComponent) true))

(fn ConstructionScript)
