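; Character part modifier Blueprint for the Party Trooper (Buffet) cosmetic that makes its idle VFX react to the wearer's health.
; After res-in it waits three seconds, caches the associated player pawn and binds to that pawn's health-changed event.
; Each health change sets the idle particle component's "Damage" parameter to a clamped inverse of health, assuming a 100-point scale, so the value rises as health drops.
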(event EventOnPostResIn
  (|CallParentFunction)
  (Utilities|FlowControl|Delay 3.0)
  (Variables|Default|SetAssociatedPlayerPawn (CustomCharacter|GetAssociatedPlayerPawn))
  (Health|AssignOnPawnHealthChanged (Variables|Default|GetAssociatedPlayerPawn) (EventDispatchers|CreateEvent :!event Custom|OnPawnHealthChanged_Event_Local)))

(event Custom|OnPawnHealthChanged_Event_Local
  (bind _associatedplayerpawn (Variables|Default|GetAssociatedPlayerPawn)
    (:then
      (Effects|Components|ParticleSystem|SetFloatParameter (CallFunction|GetIdleVFXComponent) "Damage" (Math|Float|Clamp(Float) (- 1.0 (* (FortPawn|GetHealth _associatedplayerpawn) 0.01)))))
    (:else)))

(fn ConstructionScript
  (|CallParentFunction))
