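; Part modifier for the Patches skin's Stage1 variant that burns open the character's heart as the player deals damage.
; DamageDealt stat changes grow the hole by BurnRadiusOffset steps (despite "damage received" variable names), with timelines driving material scalars, Niagara reveal/close FX and idle VFX HoleRadius.
; A repeating timer reverses the effect once the hit count stops changing.
; In the frontend, a looping timer demonstrates opening and closing.
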(event Custom|frontend_Closeheart
  (bind _returnvalue (CallFunction|GetIdleVFXComponent))
  ;[0.09815 0.5 0.2162] frontend_Close
  (if (Utilities|Enum|Equal(Enum) (CustomCharacter|GetCharacterPartType) "Body")
    (bind _returnvalue_2 (PartModifier|SpawnSystem|SpawnSystemAttachedPartModifier "/BRCosmetics/Characters/Player/Male/Medium/Bodies/M_MED_Patches/FX/NS_MED_Patches_CloseFX.NS_MED_Patches_CloseFX" (CustomCharacter|GetSkeletalMeshForPartOnPlayerPawn) "None" "0, 0, 0" "0, 0, 0" "SnapToTarget")
      (:Success
        (bind (_direction _heartfxmask) (Timeline|Frontend_Close ReverseFromEnd)
          (:Update
            (Rendering|Material|SetScalarParameterValueonMaterials (CustomCharacter|GetSkeletalMeshForPartOnPlayerPawn) "BurnRadius" _heartfxmask)
            (Rendering|Material|SetScalarParameterValueonMaterials (CustomCharacter|GetSkeletalMeshForPartOnPlayerPawn) "Hair burn Blend" _heartfxmask)
            (Utilities|IsValid _returnvalue_2
              (:Is%20Valid
                (Niagara|SetNiagaraVariable(Float) _returnvalue_2 "HeartBurn" _heartfxmask)
                (Niagara|SetNiagaraVariable(Bool) _returnvalue_2 "bIsFrontEnd" true)
                (Utilities|IsValid _returnvalue
                  (:Is%20Valid
                    (Effects|Components|ParticleSystem|SetFloatParameter _returnvalue "HoleRadius" _heartfxmask))))))
          (:Finished
            (Rendering|Material|SetScalarParameterValueonMaterials (CustomCharacter|GetSkeletalMeshForPartOnPlayerPawn) "BurnRadius" 0.0)
            (Rendering|Material|SetScalarParameterValueonMaterials (CustomCharacter|GetSkeletalMeshForPartOnPlayerPawn) "Hair burn Blend" 0.0)
            (Variables|Default|SetHeart%20is%20opened false)))))
    (else
      (Timeline|Frontend_Close ReverseFromEnd))))

(event Custom|frontend_OpenHeart
  (bind _returnvalue (CustomCharacter|GetSkeletalMeshForPartOnPlayerPawn))
  (bind _returnvalue_2 (CallFunction|GetIdleVFXComponent))
  ;[0.09815 0.5 0.2162] frontend__Open heart receive damage
  (if (Utilities|Enum|Equal(Enum) (CustomCharacter|GetCharacterPartType) "Body")
    (bind _returnvalue_3 (PartModifier|SpawnSystem|SpawnSystemAttachedPartModifier "/BRCosmetics/Characters/Player/Male/Medium/Bodies/M_MED_Patches/FX/NS_MED_Patches_RevealFX.NS_MED_Patches_RevealFX" (CustomCharacter|GetSkeletalMeshForPartOnPlayerPawn) "None" "0, 0, 0" "0, 0, 0" "SnapToTarget")
      (:Success
        (bind (_direction _heartfxmask) (Timeline|frontend_Open%20Heart PlayFromStart)
          (:Update
            (Variables|Default|SetPlaying%20open%20animation true)
            (Rendering|Material|SetScalarParameterValueonMaterials _returnvalue "BurnRadius" _heartfxmask)
            (Rendering|Material|SetScalarParameterValueonMaterials _returnvalue "Hair burn Blend" _heartfxmask)
            (Rendering|Material|SetScalarParameterValueonMaterials _returnvalue "Clothing edge Emissive Min" (+ (* (- 1.0 _heartfxmask) 2e+02) 4.0))
            (Rendering|Material|SetScalarParameterValueonMaterials _returnvalue "Hair edge Emissive Min" (+ (* (- 1.0 _heartfxmask) 1.5) 0.5))
            (Utilities|IsValid _returnvalue_3
              (:Is%20Valid
                (Niagara|SetNiagaraVariable(Float) _returnvalue_3 "HeartBurn" _heartfxmask)
                (Niagara|SetNiagaraVariable(Bool) _returnvalue_3 "bIsFrontEnd" true)
                (Utilities|IsValid _returnvalue_2
                  (:Is%20Valid
                    (Effects|Components|ParticleSystem|SetFloatParameter _returnvalue_2 "HoleRadius" _heartfxmask))))))
          (:Finished
            (Variables|Default|SetPlaying%20open%20animation false)
            (Utilities|Time|UnpauseTimerbyHandle (Variables|Default|GetDamage%20received%20Check))))))
    (else
      (Timeline|frontend_Open%20Heart PlayFromStart))))

(event EventBeginPlay
  (Variables|Default|SetNumber%20of%20damage%20received 0)
  (Variables|Default|SetLast%20Number%20of%20damage%20received 0.0)
  (Variables|Default|SetHeart%20is%20opened false)
  (Variables|Default|SetPlaying%20Close%20animation false))

(event EventOnPostResIn
  (bind _returnvalue (CustomCharacter|GetSkeletalMeshForPartOnPlayerPawn))
  (bind _returnvalue_2 (CallFunction|GetIdleVFXComponent))
  (bind _number_of_damage_received (Variables|Default|GetNumber%20of%20damage%20received))
  (bind _returnvalue_3 (- _number_of_damage_received 1))
  (bind _burnradiusoffset (Variables|Default|GetBurnRadiusOffset))
  (bind _returnvalue_4 (* _burnradiusoffset _returnvalue_3))
  (|CallParentFunction)
  (Utilities|FlowControl|Sequence
    (:then_0
      (Cosmetics|WatchCosmeticStat (CustomCharacter|GetAssociatedPlayerPawn) "(TagName=\"StatManager.DamageDealt\")" 0 false
        (:OnStatChanged
          (if (GameplayTags|HasTag (Cosmetics|GatherMetaTags (CustomCharacter|GetAssociatedPlayerPawn) "/BRCosmetics/Athena/Items/Cosmetics/Characters/Character_Patches.Character_Patches" :!class /Script/FortniteGame.FortPlayerPawn) "(TagName=\"Cosmetics.Variant.Property.Stage1\")")
            ;[0 0.1186 0.5] Open heart receive damage
            (if (not (Variables|Default|GetPlaying%20Close%20animation))
              ;;[1 1 1] protection
              (Utilities|FlowControl|Sequence
                (:then_0
                  _)
                (:then_1
                  (Utilities|FlowControl|Delay 0.3)))
              (Utilities|FlowControl|DoOnce false
                (:Completed
                  (Variables|Default|SetHeart%20is%20opened true)
                  (Variables|Default|SetNumber%20of%20damage%20received (+ (Variables|Default|GetNumber%20of%20damage%20received) 1))
                  (if (Utilities|Enum|Equal(Enum) (CustomCharacter|GetCharacterPartType) "Body")
                    (bind _returnvalue_5 (PartModifier|SpawnSystem|SpawnSystemAttachedPartModifier "/BRCosmetics/Characters/Player/Male/Medium/Bodies/M_MED_Patches/FX/NS_MED_Patches_RevealFX.NS_MED_Patches_RevealFX" (CustomCharacter|GetSkeletalMeshForPartOnPlayerPawn) "None" "0, 0, 0" "0, 0, 0" "SnapToTarget")
                      (:Success
                        (bind (_direction _heartfxmask) (Timeline|Heart%20open%20Mask PlayFromStart)
                          (:Update
                            (Variables|Default|SetPlaying%20open%20animation true)
                            (Rendering|Material|SetScalarParameterValueonMaterials _returnvalue "BurnRadius" (+ (* _heartfxmask _burnradiusoffset) _returnvalue_4))
                            (Rendering|Material|SetScalarParameterValueonMaterials _returnvalue "Hair burn Blend" (+ (* _heartfxmask _burnradiusoffset) _returnvalue_4))
                            (Rendering|Material|SetScalarParameterValueonMaterials _returnvalue "Clothing edge Emissive Min" (+ (* (- 1.0 _heartfxmask) 2e+02) 4.0))
                            (Rendering|Material|SetScalarParameterValueonMaterials _returnvalue "Hair edge Emissive Min" (+ (* (- 1.0 _heartfxmask) 1.5) 0.5))
                            (Utilities|IsValid _returnvalue_5
                              (:Is%20Valid
                                (Niagara|SetNiagaraVariable(Float) _returnvalue_5 "HeartBurn" (+ (* _heartfxmask _burnradiusoffset) _returnvalue_4))
                                (Utilities|IsValid _returnvalue_2
                                  (:Is%20Valid
                                    (Effects|Components|ParticleSystem|SetFloatParameter _returnvalue_2 "HoleRadius" (+ (* _heartfxmask _burnradiusoffset) _returnvalue_4)))))))
                          (:Finished
                            (Variables|Default|SetPlaying%20open%20animation false)
                            (Utilities|Time|UnpauseTimerbyHandle (Variables|Default|GetDamage%20received%20Check))))))
                    (else
                      (Timeline|Heart%20open%20Mask PlayFromStart))))))
            (else
              (Variables|Default|SetNumber%20of%20damage%20received 0)
              (Variables|Default|SetLast%20Number%20of%20damage%20received 0.0)
              (Variables|Default|SetHeart%20is%20opened false)
              (Variables|Default|SetPlaying%20Close%20animation false))))))
    (:then_1)
    (:then_2
      (if (GameplayTags|HasTag (Cosmetics|GatherMetaTags (CustomCharacter|GetAssociatedPlayerPawn) "/BRCosmetics/Athena/Items/Cosmetics/Characters/Character_Patches.Character_Patches" :!class /Script/FortniteGame.FortPlayerPawn) "(TagName=\"Cosmetics.Variant.Property.Stage1\")")
        ;[0.1248 0.9801 1] Close heart
        (if (not (Variables|Default|GetPlaying%20open%20animation))
          (bind _returnvalue_6 (Utilities|Time|SetTimerbyEvent (delegate Custom|Damage%20received) 15.0 true))
          (Variables|Default|SetDamage%20received%20Check _returnvalue_6))
        (else
          (Utilities|Time|PauseTimerbyHandle (Variables|Default|GetDamage%20received%20Check)))))
    (:then_3
      (if (GameplayTags|HasTag (Cosmetics|GatherMetaTags (CustomCharacter|GetAssociatedPlayerPawn) "/BRCosmetics/Athena/Items/Cosmetics/Characters/Character_Patches.Character_Patches" :!class /Script/FortniteGame.FortPlayerPawn) "(TagName=\"Cosmetics.Variant.Property.Stage1\")")
        (if (CustomCosmeticModifier|IsinFrontEnd)
          (Utilities|Time|SetTimerbyEvent (delegate Custom|frontend%20circle) 14.0 true false -12.5))))))

(event Custom|Damage%20received
  (bind _returnvalue (CallFunction|GetIdleVFXComponent))
  (bind _number_of_damage_received (Variables|Default|GetNumber%20of%20damage%20received))
  (bind _returnvalue_2 (Math|Conversions|ToFloat(Integer) _number_of_damage_received))
  (bind _returnvalue_3 (CallFunction|GetIdleVFXComponent))
  (bind _burnradiusoffset (Variables|Default|GetBurnRadiusOffset))
  (bind _returnvalue_4 (* _returnvalue_2 _burnradiusoffset))
  (bind _burnradiusoffset_2 (Variables|Default|GetBurnRadiusOffset))
  (bind _maximum_number_of_damage_received (Variables|Default|GetMaximum%20number%20of%20damage%20received))
  (bind _returnvalue_5 (* _maximum_number_of_damage_received _burnradiusoffset_2))
  ;[0.1248 0.9801 1] Close heart
  (if (== (Variables|Default|GetLast%20Number%20of%20damage%20received) (Variables|Default|GetNumber%20of%20damage%20received))
    (if (Variables|Default|GetHeart%20is%20opened)
      (if (Variables|Default|GetPlaying%20Close%20animation)
        _
        (elif (Utilities|Enum|Equal(Enum) (CustomCharacter|GetCharacterPartType) "Body")
          (bind _returnvalue_6 (PartModifier|SpawnSystem|SpawnSystemAttachedPartModifier "/BRCosmetics/Characters/Player/Male/Medium/Bodies/M_MED_Patches/FX/NS_MED_Patches_CloseFX.NS_MED_Patches_CloseFX" (CustomCharacter|GetSkeletalMeshForPartOnPlayerPawn) "None" "0, 0, 0" "0, 0, 0" "SnapToTarget")
            (:Success
              (bind (_direction _heartfxmask) (Timeline|Heart%20Close%20Mask ReverseFromEnd)
                (:Update
                  (if (>= (Variables|Default|GetNumber%20of%20damage%20received) (Variables|Default|GetMaximum%20number%20of%20damage%20received))
                    (Variables|Default|SetPlaying%20Close%20animation true)
                    (Rendering|Material|SetScalarParameterValueonMaterials (CustomCharacter|GetSkeletalMeshForPartOnPlayerPawn) "BurnRadius" (* _heartfxmask _returnvalue_5))
                    (Rendering|Material|SetScalarParameterValueonMaterials (CustomCharacter|GetSkeletalMeshForPartOnPlayerPawn) "Hair burn Blend" (* _heartfxmask _returnvalue_5))
                    (Utilities|IsValid _returnvalue_6
                      (:Is%20Valid
                        (Niagara|SetNiagaraVariable(Float) _returnvalue_6 "HeartBurn" (* _heartfxmask _returnvalue_5))
                        (Utilities|IsValid _returnvalue_3
                          (:Is%20Valid
                            (Effects|Components|ParticleSystem|SetFloatParameter _returnvalue_3 "HoleRadius" (* _heartfxmask _returnvalue_5))))))
                    (else
                      (Variables|Default|SetPlaying%20Close%20animation true)
                      (Rendering|Material|SetScalarParameterValueonMaterials (CustomCharacter|GetSkeletalMeshForPartOnPlayerPawn) "BurnRadius" (* _heartfxmask _returnvalue_4))
                      (Rendering|Material|SetScalarParameterValueonMaterials (CustomCharacter|GetSkeletalMeshForPartOnPlayerPawn) "Hair burn Blend" (* _heartfxmask _returnvalue_4))
                      (Utilities|IsValid _returnvalue_6
                        (:Is%20Valid
                          (Niagara|SetNiagaraVariable(Float) _returnvalue_6 "HeartBurn" (* _heartfxmask _returnvalue_4))
                          (Utilities|IsValid _returnvalue
                            (:Is%20Valid
                              (Effects|Components|ParticleSystem|SetFloatParameter _returnvalue "HoleRadius" (* _heartfxmask _returnvalue_4)))))))))
                (:Finished
                  (Rendering|Material|SetScalarParameterValueonMaterials (CustomCharacter|GetSkeletalMeshForPartOnPlayerPawn) "BurnRadius" 0.0)
                  (Rendering|Material|SetScalarParameterValueonMaterials (CustomCharacter|GetSkeletalMeshForPartOnPlayerPawn) "Hair burn Blend" 0.0)
                  (Variables|Default|SetHeart%20is%20opened false)
                  (Variables|Default|SetNumber%20of%20damage%20received 0)
                  (Variables|Default|SetLast%20Number%20of%20damage%20received 0.0)
                  (Variables|Default|SetPlaying%20Close%20animation false)
                  (Utilities|Time|PauseTimerbyHandle (Variables|Default|GetDamage%20received%20Check))))))
          (else
            (Timeline|Heart%20Close%20Mask ReverseFromEnd)))))
    (else
      (Variables|Default|SetLast%20Number%20of%20damage%20received (Math|Conversions|ToFloat(Integer) (Variables|Default|GetNumber%20of%20damage%20received))))))

(event Custom|frontend%20circle
  (CallFunction|FrontendOpenHeart)
  (Utilities|FlowControl|Delay 8.0)
  (CallFunction|FrontendCloseheart))

(fn ConstructionScript)
