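; Cosmetic part modifier for the TargetPractice male character part that triggers reactive lights when the player's kill stat changes.
; On post-res-in, it resets KillCount and watches the AthenaKills stat.
; Each change stores the count, zeroes TP_KillsInput and stamps game time on the mesh materials.
; Off mobile, it also copies the timestamp and slot 0 material timing parameters to the idle Niagara VFX, sharing timing values between both.
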(event EventOnPostResIn
  (bind _returnvalue (CustomCharacter|GetSkeletalMeshForPartOnPlayerPawn))
  (|CallParentFunction)
  (Variables|Default|SetKillCount 0.0)
  (bind (_statname _statvalue) (Cosmetics|WatchCosmeticStat (CustomCharacter|GetAssociatedPlayerPawn) "(TagName=\"StatManager.AthenaKills\")")
    (:OnStatChanged
      (Variables|Default|SetKillCount (Math|Conversions|ToFloat(Integer) _statvalue))
      ;[1 1 1] Set Mesh Materials
      (Rendering|Material|SetScalarParameterValueonMaterials _returnvalue "TP_KillsInput" 0.0)
      (Rendering|Material|SetScalarParameterValueonMaterials _returnvalue "TP_GameTimeFromBP" (Utilities|Time|GetGameTimeinSeconds))
      (if (not (Fortnite|IsMobilePlatform))
        (Utilities|IsValid (CallFunction|GetIdleVFXComponent)
          (:Is%20Valid
            ;[1 1 1] Passing material instance values to the Niagara lights
            (bind _asmaterial_instance_dynamic (Utilities|Casting|CastToMaterialInstanceDynamic (Rendering|Material|GetMaterial (CustomCharacter|GetSkeletalMeshForPartOnPlayerPawn) 0)))
            (bind _returnvalue_2 (Rendering|Material|GetScalarParameterValue _asmaterial_instance_dynamic "TP_TimeOffset" :!class /Script/Engine.MaterialInstanceDynamic))
            (Effects|Components|ParticleSystem|SetFloatParameter (CallFunction|GetIdleVFXComponent) "User.TP_TimeOffset" _returnvalue_2)
            (bind _returnvalue_3 (Rendering|Material|GetScalarParameterValue _asmaterial_instance_dynamic "TP_UseVertexColorChannelForTimeOFfset" :!class /Script/Engine.MaterialInstanceDynamic))
            (Effects|Components|ParticleSystem|SetFloatParameter (CallFunction|GetIdleVFXComponent) "User.TP_UseVertexColorChannelForTimeOFfset" _returnvalue_3)
            (Effects|Components|ParticleSystem|SetFloatParameter (CallFunction|GetIdleVFXComponent) "User.TP_GameTimeFromBP" (Utilities|Time|GetGameTimeinSeconds))
            (bind _returnvalue_4 (Rendering|Material|GetScalarParameterValue _asmaterial_instance_dynamic "TP_ReactiveGlowDuration" :!class /Script/Engine.MaterialInstanceDynamic))
            (Effects|Components|ParticleSystem|SetFloatParameter (CallFunction|GetIdleVFXComponent) "User.TP_ReactiveGlowDuration" _returnvalue_4)
            (bind _returnvalue_5 (Rendering|Material|GetScalarParameterValue _asmaterial_instance_dynamic "TP_ReactiveBlinkDuration" :!class /Script/Engine.MaterialInstanceDynamic))
            (Effects|Components|ParticleSystem|SetFloatParameter (CallFunction|GetIdleVFXComponent) "User.TP_ReactiveBlinkDuration" _returnvalue_5)
            (bind _returnvalue_6 (Rendering|Material|GetScalarParameterValue _asmaterial_instance_dynamic "TP_LightsDefaultSpeed" :!class /Script/Engine.MaterialInstanceDynamic))
            (Effects|Components|ParticleSystem|SetFloatParameter (CallFunction|GetIdleVFXComponent) "User.TP_LightsDefaultSpeed" _returnvalue_6)
            (bind _returnvalue_7 (Rendering|Material|GetScalarParameterValue _asmaterial_instance_dynamic "TP_LightsReactiveSpeed" :!class /Script/Engine.MaterialInstanceDynamic))
            (Effects|Components|ParticleSystem|SetFloatParameter (CallFunction|GetIdleVFXComponent) "User.TP_LightsReactiveSpeed" _returnvalue_7)
            (bind _returnvalue_8 (Rendering|Material|GetScalarParameterValue _asmaterial_instance_dynamic "TP_ValueToHideonReactive" :!class /Script/Engine.MaterialInstanceDynamic))
            (Effects|Components|ParticleSystem|SetFloatParameter (CallFunction|GetIdleVFXComponent) "User.TP_ValueToHideonReactive" _returnvalue_8)))))))

(fn GameTimeFromBPForGlow
  (params (A :type Float%20%28double-precision%29) (B :type Float%20%28double-precision%29))
  (return 0.0))

(fn ConstructionScript)
