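; Cosmetic part modifier for The Herald's head that drives a simulated feather-hair material with two progressive variants.
; On post-init it creates the variant's dynamic material and binds the idle Niagara system to itself and the player's body mesh.
; Niagara particle exports feed the feather material's force, spin and rotation, while a timer supplies walking movement state.
; Front-end, team and platform checks gate the simulation; when disabled, the dark variant tracks head position by timer.
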(event EventOnPostResIn
  (|CallParentFunction)
  (Utilities|FlowControl|Sequence
    (:then_0
      (bind _index (Class|FortCosmeticFunctionLibrary|GetVariantChannelActiveIndexForSlot "(TagName=\"LoadoutSlot.Character\")" (CosmeticLoadout|GetCosmeticLoadout) "(TagName=\"Cosmetics.Variant.Channel.Progressive\")" (Variables|Default|GetVariant%20List)))
      (Variables|Default|SetVariantIndex _index)
      (Variables|Default|SetSwapBool (== _index 0))
      (bind _returnvalue (Rendering|Material|CreateDynamicMaterialInstance (Utilities|Select :Index _index :Option%200 "/BRCosmetics/Characters/Player/Female/Medium/Bodies/F_MED_TheHerald/FX/HairMaterial/M_TheHerald_FeatherSim_Inst.M_TheHerald_FeatherSim_Inst" :Option%201 "/BRCosmetics/Characters/Player/Female/Medium/Bodies/F_MED_TheHerald/FX/HairMaterial/M_TheHerald_FeatherSim_Dark_Inst.M_TheHerald_FeatherSim_Dark_Inst") "None" "None" :!class /Script/Engine.KismetMaterialLibrary))
      (bind _output_get (Variables|Default|SetHeadMaterial _returnvalue))
      (Rendering|Material|SetMaterial (CustomCharacter|GetSkeletalMeshForPartOnPlayerPawn) 0 _output_get)
      (bind _asniagara_particle_system_component (Utilities|Casting|CastToNiagaraComponent (CallFunction|GetIdleVFXComponent)))
      (Niagara|SetNiagaraVariable(Object) _asniagara_particle_system_component "TargetBP" self)
      (Niagara|SetNiagaraSkeletalMeshComponent _asniagara_particle_system_component "PlayerBody" (Customization|GetSkeletalMeshforPartType (CustomCharacter|GetAssociatedPlayerPawn) "Body" :!class /Script/FortniteGame.FortPlayerPawn))
      (Variables|Default|SetTargetNS _asniagara_particle_system_component)
      (Effects|Components|ParticleSystem|SetBoolParameter (Variables|Default|GetTargetNS) "bVariant" (Variables|Default|GetSwapBool)))
    (:then_1
      (if (CustomCosmeticModifier|IsinFrontEnd)
        _
        (elif (Team|OnLocalPlayersTeam (CustomCharacter|GetAssociatedPlayerPawn))
          (if (Fortnite|IsMobilePlatform)
            (Development|PrintString "TheHerald:MobilePlatform" true true "(R=1.000000,G=0.000000,B=0.000000,A=1.000000)" 1e+01)
            (Development|PrintString "TheHerald:NotOnLocalPlayersTeam,DisableNiagaraSim" true true "(R=1.000000,G=0.000000,B=0.000000,A=1.000000)" 1e+01)
            (Utilities|FlowControl|Sequence
              (:then_0
                (bind _asniagara_particle_system_component_2 (Utilities|Casting|CastToNiagaraComponent (CallFunction|GetIdleVFXComponent)))
                (Effects|Components|ParticleSystem|SetBoolParameter _asniagara_particle_system_component_2 "DisableSim" true))
              (:then_1
                (Rendering|Material|SetScalarParameterValue (Variables|Default|GetHeadMaterial) "DisableSim" 1.0 :!class /Script/Engine.MaterialInstanceDynamic))
              (:then_2
                (if (== (Variables|Default|GetVariantIndex) 1)
                  (Development|PrintString "TheHerald:EnableHeadLocationTimer" true true "(R=1.000000,G=0.000000,B=0.000000,A=1.000000)" 1e+01)
                  (Utilities|DoOncePerFrame
                    (:%20%20
                      (bind _returnvalue_2 (Utilities|Time|SetTimerbyEvent (delegate Custom|SetDisableSimHeadCenter) 0.001 true)) ; JIRA FORT-529533 changed the timer time to 0.001 to fix the flicker issue
                      (Variables|Default|SetHeadCenterTimer _returnvalue_2))))))
            (else
              _))
          (else
            _)))
      (Development|PrintString "TheHerald:OnLocalPlayersTeam,EnableNiagaraSim" true true "(R=0.319449,G=1.000000,B=0.000000,A=1.000000)" 1e+01)
      (Rendering|Material|SetScalarParameterValue (Variables|Default|GetHeadMaterial) "DisableSim" 0.0 :!class /Script/Engine.MaterialInstanceDynamic)
      (Utilities|DoOncePerFrame
        (:%20%20
          (bind _returnvalue_3 (Utilities|Time|SetTimerbyEvent (delegate Custom|CheckIsMove) 0.05 true))
          (Variables|Default|SetCheckMoveTimer _returnvalue_3))))))

(event EventPreDestruction
  (bind _checkmovetimer (Variables|Default|GetCheckMoveTimer))
  (bind _headcentertimer (Variables|Default|GetHeadCenterTimer))
  (|CallParentFunction)
  (Utilities|FlowControl|Sequence
    (:then_0
      (if (Utilities|Time|IsValidTimerHandle _checkmovetimer)
        (Utilities|Time|ClearandInvalidateTimerbyHandle _checkmovetimer)))
    (:then_1
      (if (Utilities|Time|IsValidTimerHandle _headcentertimer)
        (Utilities|Time|ClearandInvalidateTimerbyHandle _headcentertimer)))))

(event Custom|CheckIsMove
  (Variables|Default|SetbIsMove (and (> (Math|Vector|VectorLength (Pawn|Components|CharacterMovement|GetLastUpdateVelocity (Variables|/Script/Engine.Character|GetCharacterMovement (CustomCharacter|GetAssociatedPlayerPawn)))) 2e+02) (Pawn|Components|CharacterMovement|IsWalking (Variables|/Script/Engine.Character|GetCharacterMovement (CustomCharacter|GetAssociatedPlayerPawn))))))

(event Custom|SetDisableSimHeadCenter
  (Rendering|Material|SetVectorParameterValue (Variables|Default|GetHeadMaterial) "DisableSim_HeadCenter" (Math|Conversions|ToLinearColor(Vector) (Transformation|GetSocketLocation (Variables|/Script/Engine.Character|GetMesh (CustomCharacter|GetAssociatedPlayerPawn)) "head")) :!class /Script/Engine.MaterialInstanceDynamic))

(event Niagara|EventReceiveParticleData
  (params (Data :type Array%20of%20Basic%20Particle%20Data%20Structures) (NiagaraSystem :type Niagara%20System%20Object%20Reference) (SimulationPositionOffset :type Vector%20%28by%20ref%29 :default "0, 0, 0"))
  (bind _output (Utilities|Array|Get(acopy) Data))
  (bind (_position _size _velocity) (Utilities|Struct|BreakBasicParticleData _output))
  (bind _headmaterial (Variables|Default|GetHeadMaterial))
  (Rendering|Material|SetVectorParameterValue _headmaterial "Feather_Force" (Utilities|Struct|MakeLinearColor (.x _velocity) (.y _velocity) (.z _velocity)) :!class /Script/Engine.MaterialInstanceDynamic)
  (Rendering|Material|SetScalarParameterValue _headmaterial "AddSpin" _size :!class /Script/Engine.MaterialInstanceDynamic)
  (Niagara|SetNiagaraVariable(Bool) (Variables|Default|GetTargetNS) "IsMove" (Variables|Default|GetbIsMove))
  (Rendering|Material|SetVectorParameterValue _headmaterial "Quat" (Utilities|Struct|MakeLinearColor (.x _position) (.y _position) (.z _position)) :!class /Script/Engine.MaterialInstanceDynamic))

(fn ConstructionScript
  (|CallParentFunction))
