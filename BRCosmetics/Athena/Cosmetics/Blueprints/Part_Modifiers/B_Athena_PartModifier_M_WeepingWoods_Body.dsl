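; Cosmetic part modifier for the male medium Weeping Woods character body that plays a particle effect when the pawn lands.
; After calling the parent landed event, it spawns a debounced, auto-destroying legacy emitter attached to the root of the part's skeletal mesh, matching the pawn's custom depth.
; On a successful spawn, it applies the player's character loadout slot variants to the new component so the effect follows the selected style.
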(event EventOnPawnLanded
  (|CallParentFunction)
  (bind _returnvalue (PartModifier|SpawnSystem|SpawnEmitterAttachedPartModifier(Legacy) :EmitterTemplate "/BRCosmetics/Characters/Player/Male/Medium/Bodies/M_MED_WeepingWoods/FX/P_M_WeepingWoods_2.P_M_WeepingWoods_2" :AttachToComponent (CustomCharacter|GetSkeletalMeshForPartOnPlayerPawn) :AttachPointName "root" :Scale "1.000000,1.000000,1.000000" :LocationType "SnapToTargetIncludingScale" :bAutoDestroy true :PoolingMethod "None" :bAutoActivate true :bMatchPawnCustomDepth true :bDebounce true :DebounceTime -1.0)
    (:Success
      (Class|FortCosmeticFunctionLibrary|ApplyVariantstoComponentbySlotTag (CustomCharacter|GetAssociatedPlayerPawn) "(TagName=\"LoadoutSlot.Character\")" _returnvalue))))

(fn ConstructionScript)
