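; Cosmetic part modifier for the Madame Moth skin, handling front-end hologram materials and emote-reactive idle VFX.
; Each tick, when a front-end pawn is a hologram using the Posh body mesh, it sets material slots 1 and 2 to a blank material.
; After res-in it caches the part's anim instance.
; Emote begin and end toggle the "Emote begin" bool parameter on the idle VFX component.
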(event EventTick
  (params (DeltaSeconds :type Float%20%28single-precision%29))
  (bind _returnvalue (CustomCharacter|GetSkeletalMeshForPartOnPlayerPawn))
  (if (CustomCosmeticModifier|IsinFrontEnd)
    (if (Variables|/Script/FortniteGame.FortPlayerPawn|GetbIsInFrontEndHologram (CustomCharacter|GetAssociatedPlayerPawn))
      (bind _asskeletal_mesh_component (Utilities|Casting|CastToSkeletalMeshComponent _returnvalue))
      (if (== (Variables|/Script/Engine.SkinnedMeshComponent|GetSkeletalMesh _asskeletal_mesh_component) "/BRCosmetics/Characters/Player/Female/Medium/Bodies/F_MED_MadameMoth_Posh/Meshes/F_MED_MadameMoth_Posh.F_MED_MadameMoth_Posh")
        (Rendering|Material|SetMaterial _returnvalue 1 "/Game/Accessories/FORT_Backpacks/M_MED_Celestial_Backpack/M_Blank.M_Blank")
        (Rendering|Material|SetMaterial _returnvalue 2 "/Game/Accessories/FORT_Backpacks/M_MED_Celestial_Backpack/M_Blank.M_Blank")))))

(event EventOnPostResIn
  (|CallParentFunction)
  (bind _ascustom_character_part_anim_instance (Utilities|Casting|CastToCustomCharacterPartAnimInstance (Components|SkeletalMesh|GetAnimInstance (Customization|GetSkeletalMeshforPartType (CustomCharacter|GetAssociatedPlayerPawn) (CustomCharacter|GetCharacterPartType) :!class /Script/FortniteGame.FortPlayerPawn))))
  (Variables|Default|SetCustom%20Character%20Part%20Anim%20Instance _ascustom_character_part_anim_instance))

(event EventOnBeginEmote
  (params (MontageItemDef :type Fort%20Item%20Definition%20Object%20Reference) (PawnEmoting :type Fort%20Pawn%20Object%20Reference))
  (Effects|Components|ParticleSystem|SetBoolParameter (CallFunction|GetIdleVFXComponent) "Emote begin" true))

(event EventOnEndEmote
  (params (MontageItemDef :type Fort%20Item%20Definition%20Object%20Reference) (PawnEmoting :type Fort%20Pawn%20Object%20Reference))
  (Effects|Components|ParticleSystem|SetBoolParameter (CallFunction|GetIdleVFXComponent) "Emote begin" false))

(fn ConstructionScript
  (|CallParentFunction))
