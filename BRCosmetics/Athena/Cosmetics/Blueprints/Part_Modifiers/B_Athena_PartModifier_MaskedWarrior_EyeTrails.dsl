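; Cosmetic character part modifier for Masked Warrior that adds an eye trail mesh to the player pawn.
; On the post res-in event it runs the parent logic, then attaches the SM_MaskedWarrior_EyeTrail component to the "head" socket of the pawn's part skeletal mesh.
; The attachment keeps relative location, rotation and scale, with welding enabled.
; The construction script only calls the parent.
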(event EventOnPostResIn
  (|CallParentFunction)
  (Transformation|AttachComponentToComponent (Variables|Default|GetSM_MaskedWarrior_EyeTrail) (CustomCharacter|GetSkeletalMeshForPartOnPlayerPawn) "head" "KeepRelative" "KeepRelative" "KeepRelative" true))

(fn ConstructionScript
  (|CallParentFunction))
