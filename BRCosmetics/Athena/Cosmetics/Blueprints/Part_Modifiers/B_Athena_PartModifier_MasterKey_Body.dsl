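; Body cosmetic part modifier for the MasterKey character that drives its idle VFX, including lights tied to the day phase.
; On day phase changes it switches on the phase to set the idle particle system's LightSpawnRate, with Evening as the light-on case and a zero rate for hiding, printing debug strings.
; After res-in it passes the associated player pawn to the particles as BoneSocketActor.
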(event EventOnDayPhaseChanged
  (params (CurrentDayPhase :type EFortDayPhase%20Enum :default "Morning") (PreviousDayPhase :type EFortDayPhase%20Enum :default "Morning") (bAtCreation :type Boolean))
  (bind _returnvalue (CallFunction|GetIdleVFXComponent))
  (|CallParentFunction CurrentDayPhase PreviousDayPhase bAtCreation)
  (switch Utilities|FlowControl|Switch|SwitchonEFortDayPhase CurrentDayPhase
    (:Morning
      _)
    (:Day
      _)
    (:Evening
      (Effects|Components|ParticleSystem|SetFloatParameter _returnvalue "LightSpawnRate" 1.5)
      (Development|PrintString "ShowLights"))
    (:Night
      _))
  (Effects|Components|ParticleSystem|SetFloatParameter _returnvalue "LightSpawnRate" 0.0)
  (Development|PrintString "HideLights"))

(event EventOnPostResIn
  (|CallParentFunction)
  (Effects|Components|ParticleSystem|SetActorParameter (CallFunction|GetIdleVFXComponent) "BoneSocketActor" (CustomCharacter|GetAssociatedPlayerPawn)))

(fn ConstructionScript)
