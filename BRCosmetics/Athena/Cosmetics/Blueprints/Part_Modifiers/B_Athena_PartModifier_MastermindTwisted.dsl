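; Character part modifier Blueprint for the MastermindTwisted cosmetic that adjusts its idle particle effect to the pawn's movement state.
; Jump, land and skydive begin/end events each call the parent, then set the "Velocity" vector parameter on the idle VFX component.
; Jumping or skydiving zeroes the parameter, while landing or ending a skydive restores a small non-zero vector.
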(event EventOnPawnJumped
  (|CallParentFunction)
  (Effects|Components|ParticleSystem|SetVectorParameter (CallFunction|GetIdleVFXComponent) "Velocity" "0, 0, 0"))

(event EventOnPawnLanded
  (|CallParentFunction)
  (Effects|Components|ParticleSystem|SetVectorParameter (CallFunction|GetIdleVFXComponent) "Velocity" "0.100000,0.100000,0.200000"))

(event EventOnBeginSkydiving
  (|CallParentFunction)
  (Effects|Components|ParticleSystem|SetVectorParameter (CallFunction|GetIdleVFXComponent) "Velocity" "0, 0, 0"))

(event EventOnEndSkydiving
  (|CallParentFunction)
  (Effects|Components|ParticleSystem|SetVectorParameter (CallFunction|GetIdleVFXComponent) "Velocity" "0.100000,0.100000,0.200000"))

(fn ConstructionScript)
