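; Part modifier for the Mechanical Engineer cosmetic, adding backpack spark and skydiving trail effects and linking kills to its pet.
; Post res-in creates or restarts world sparks, caches the player's pet, and watches the kill stat to fire the pet's KillEvent outside the front end.
; Skydiving restarts sparks, then after a short delay begins socket-based left and right air trails.
; Ending skydiving deactivates the right trail but only ends trails on the left.
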(event EventOnPostResIn
  (|CallParentFunction)
  (Utilities|FlowControl|Sequence
    (:then_0
      (Utilities|FlowControl|Sequence
        (:then_0
          (Utilities|IsValid (Variables|Default|GetWorld%20Sparks)
            (:Is%20Valid
              (Components|Activation|Activate (Variables|Default|GetWorld%20Sparks) true))
            (:Is%20Not%20Valid
              (bind _returnvalue (PartModifier|SpawnSystem|SpawnEmitterAttachedPartModifier(Legacy) "/BRCosmetics/Accessories/FORT_Backpacks/F_MED_Mechanical_Engineer_Backpack/FX/P_Backpack_MED_Mechanical_Engineer_SparkWorld.P_Backpack_MED_Mechanical_Engineer_SparkWorld" (CustomCharacter|GetSkeletalMeshForPartOnPlayerPawn) "root")
                (:Success
                  (Variables|Default|SetWorld%20Sparks _returnvalue))))))
        (:then_1
          (bind (_statname _statvalue) (Cosmetics|WatchCosmeticStat (CustomCharacter|GetAssociatedPlayerPawn) "(TagName=\"StatManager.AthenaKills\")")
            (:OnStatChanged
              (if (> _statvalue 0)
                (if (not (CustomCosmeticModifier|IsinFrontEnd))
                  (Class|BPetMechanicalEngineer|KillEvent (Variables|Default|GetPet)))))))))
    (:then_1
      (bind _asb_pet_mechanical_engineer (Utilities|Casting|CastToB_Pet_MechanicalEngineer (Variables|/Script/FortniteGame.FortPlayerPawn|GetCosmeticPetInstance (CustomCharacter|GetAssociatedPlayerPawn))))
      (Variables|Default|SetPet _asb_pet_mechanical_engineer))))

(event EventOnBeginSkydiving
  (|CallParentFunction)
  (Utilities|IsValid (Variables|Default|GetWorld%20Sparks)
    (:Is%20Valid
      (Components|Activation|Activate (Variables|Default|GetWorld%20Sparks) true))
    (:Is%20Not%20Valid
      (bind _returnvalue (PartModifier|SpawnSystem|SpawnEmitterAttachedPartModifier(Legacy) "/BRCosmetics/Accessories/FORT_Backpacks/F_MED_Mechanical_Engineer_Backpack/FX/P_Backpack_MED_Mechanical_Engineer_SparkWorld.P_Backpack_MED_Mechanical_Engineer_SparkWorld" (CustomCharacter|GetSkeletalMeshForPartOnPlayerPawn) "root")
        (:Success
          _)
        (:Failure
          _))
      (Variables|Default|SetWorld%20Sparks _returnvalue)))
  (Utilities|FlowControl|Delay 0.2)
  (Utilities|IsValid (Variables|Default|GetAirTrail_L)
    (:Is%20Valid
      (Components|Activation|Activate (Variables|Default|GetAirTrail_L) false))
    (:Is%20Not%20Valid
      (bind _returnvalue_2 (PartModifier|SpawnSystem|SpawnEmitterAttachedPartModifier(Legacy) "/BRCosmetics/Gadgets/Assets/VinderTech_GliderChute/Glider_Mechanical_Engineer_Female/FX/P_Glider_Mechanical_Engineer_AirTrail.P_Glider_Mechanical_Engineer_AirTrail")
        (:Success
          _)
        (:Failure
          _))
      (Variables|Default|SetAirTrail_L _returnvalue_2)))
  (Utilities|IsValid (Variables|Default|GetAirTrail_R)
    (:Is%20Valid
      (Components|Activation|Activate (Variables|Default|GetAirTrail_R) false))
    (:Is%20Not%20Valid
      (bind _returnvalue_3 (PartModifier|SpawnSystem|SpawnEmitterAttachedPartModifier(Legacy) "/BRCosmetics/Gadgets/Assets/VinderTech_GliderChute/Glider_Mechanical_Engineer_Female/FX/P_Glider_Mechanical_Engineer_AirTrail.P_Glider_Mechanical_Engineer_AirTrail")
        (:Success
          _)
        (:Failure
          _))
      (Variables|Default|SetAirTrail_R _returnvalue_3)))
  (Effects|Particles|Trails|BeginTrails (Variables|Default|GetAirTrail_L) "FX_TrailL1" "FX_TrailL2" "ETrailWidthMode_FromCentre" 1.0)
  (Effects|Particles|Trails|BeginTrails (Variables|Default|GetAirTrail_R) "FX_TrailR1" "FX_TrailR2" "ETrailWidthMode_FromCentre" 1.0))

(event EventOnEndSkydiving
  (|CallParentFunction)
  (Components|Activation|Deactivate (Variables|Default|GetAirTrail_R))
  (Effects|Particles|Trails|EndTrails (Variables|Default|GetAirTrail_L)))

(fn ConstructionScript)
