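; Cosmetic part modifier for the Meteorwomen suit that plays a material hit reaction when the wearing pawn takes damage.
; When the pawn and part change, it watches the DamageReceived stat on the associated pawn.
; Each stat change converts the last OnHitPawn location to mesh local space, substituting a fixed raised point if unset or too distant.
; It feeds that vector and game time to the part's materials, then a timeline animates HitAmountMulti.
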(event EventOnPawnAndPartChanged
  (bind _returnvalue (CustomCharacter|GetSkeletalMeshForPartOnPlayerPawn))
  (bind _returnvalue_2 (Transformation|GetWorldTransform _returnvalue))
  (|CallParentFunction)
  (Cosmetics|WatchCosmeticStat (CustomCharacter|GetAssociatedPlayerPawn) "(TagName=\"StatManager.DamageReceived\")" 0 false
    (:OnStatChanged
      (if (or (Utilities|Operators|Equal(==) :A (.Location (Collision|BreakHitResult (stale EventOnHitPawn Hit)))) (>= (Math|Vector|VectorLength (Math|Transform|InverseTransformLocation _returnvalue_2 (.Location (Collision|BreakHitResult (stale EventOnHitPawn Hit))))) 155.0))
        (Rendering|Material|SetVectorParameterValueonMaterials _returnvalue "HitVector" "0.000000,0.000000,75.000000")
        (else
          (Rendering|Material|SetVectorParameterValueonMaterials _returnvalue "HitVector" (Math|Transform|InverseTransformLocation _returnvalue_2 (.Location (Collision|BreakHitResult (stale EventOnHitPawn Hit)))))))
      (Rendering|Material|SetScalarParameterValueonMaterials _returnvalue "GameTimeFromBP" (Utilities|Time|GetGameTimeinSeconds))
      (bind (_direction _hitamountmulti) (Timeline|HitAmountMulti PlayFromStart)
        (:Update
          (Rendering|Material|SetScalarParameterValueonMaterials _returnvalue "HitAmountMulti" _hitamountmulti))))))

(event EventOnHitPawn
  (params (SelfActor :type Actor%20Object%20Reference) (OtherActor :type Actor%20Object%20Reference) (NormalImpulse :type Vector :default "0, 0, 0") (Hit :type Hit%20Result%20Structure%20%28by%20ref%29))
  (|CallParentFunction SelfActor OtherActor NormalImpulse Hit))

(fn ConstructionScript
  (|CallParentFunction))
