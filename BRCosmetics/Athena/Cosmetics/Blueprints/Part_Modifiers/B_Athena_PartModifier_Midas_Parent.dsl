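; Parent part modifier for Midas cosmetics that grows gold across the character and applies a "Midas touch" wrap to weapons and vehicles.
; With the ReactiveOn meta tag, kill stat changes drive staged material parameters, with game time stamped to animate transitions.
; The front end instead previews stages on a looping timer.
; The override wrap is applied through the player controller, then cleared on dropped pickups and when the modifier is destroyed.
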(event Custom|CustomEvent
  (bind _gm_currentstage (Variables|Default|GetGM_CurrentStage))
  ;[0 0.2896 1 0.5] Frontend Preview
  (if (< (Variables|Default|GetGM_CurrentStage) (+ (Variables|Default|GetGM_EndStage) 1))
    (CallFunction|TransitionLevel (Variables|Default|GetGM_CurrentStage)))
  (Math|Integer|IncrementInt _gm_currentstage
    (:%20%20
      (if (> _gm_currentstage (+ (Variables|Default|GetGM_EndStage) 1))
        (Utilities|Time|ClearandInvalidateTimerbyHandle (stale _returnvalue))
        (Variables|Default|SetGM_CurrentStage 1)
        (CallFunction|TransitionLevel 0)
        (Utilities|FlowControl|Delay 3.0)
        (bind _returnvalue (Utilities|Time|SetTimerbyEvent (delegate Custom|CustomEvent) 4.0 true))
        (Variables|Default|SetFE_Loop_TimerHandle _returnvalue)))))

(event EventPreDestruction
  (bind _returnvalue (CustomCharacter|GetAssociatedPlayerPawn))
  (|CallParentFunction)
  (Utilities|Time|ClearandInvalidateTimerbyHandle (Variables|Default|GetFE_Loop_TimerHandle))
  ; Clear wrap when the part modifier is destroyed to ensure they are removed when switching character in the phone booth
  (Utilities|IsValid _returnvalue
    (:Is%20Valid
      (if (not (FortPawn|IsDead _returnvalue))
        (Class|FortPawn|ClearOverrideWrapfromItems _returnvalue (Variables|Default|GetWeaponWrapToUse))))))

(event EventOnPostResIn
  (bind _returnvalue (CustomCharacter|GetSkeletalMeshForPartOnPlayerPawn))
  (bind _gm_endstage (Variables|Default|GetGM_EndStage))
  (bind _returnvalue_2 (+ _gm_endstage 1))
  (bind _gm_laststage (Variables|Default|GetGM_LastStage))
  ;[1 1 1] GoldMaskGrowingLogic
  (|CallParentFunction)
  (if (GameplayTags|HasTag (Cosmetics|GatherMetaTags (CustomCharacter|GetAssociatedPlayerPawn) "/BRCosmetics/Athena/Items/Cosmetics/Characters/Character_AuraCop_Bullet.Character_AuraCop_Bullet" :!class /Script/FortniteGame.FortPlayerPawn) "(TagName=\"Cosmetics.MetaTags.ReactiveOn\")")
    (if (CustomCosmeticModifier|IsinFrontEnd)
      ;[0 0.2896 1 0.5] Frontend Preview
      (Variables|Default|SetGM_CurrentStage 1)
      (CallFunction|TransitionLevel 0)
      (Utilities|FlowControl|Delay 3.0)
      (bind _returnvalue_3 (Utilities|Time|SetTimerbyEvent (delegate Custom|CustomEvent) 4.0 true))
      (Variables|Default|SetFE_Loop_TimerHandle _returnvalue_3)
      (else
        (bind (_statname _statvalue) (Cosmetics|WatchCosmeticStat (CustomCharacter|GetAssociatedPlayerPawn) "(TagName=\"StatManager.AthenaKills\")")
          (:OnStatChanged
            (Variables|Default|SetGM_LastStage (Math|Integer|SelectInt (Math|Integer|Max(Integer) _statvalue _gm_laststage) _gm_endstage (< (Math|Integer|Max(Integer) _statvalue _gm_laststage) _returnvalue_2)))
            (Utilities|FlowControl|Sequence
              (:then_0
                ;; Reactivity Per Stage look
                (CallFunction|InitializeVariabels)
                (Rendering|Material|SetScalarParameterValueonMaterials _returnvalue "GM_CurrentStage" (Utilities|Array|Get(aref) (Variables|Default|GetGM_ProgessPerStage) (Variables|Default|GetGM_LastStage)))
                (Rendering|Material|SetVectorParameterValueonMaterials _returnvalue "GM_TransitionNoise_Tiling" (Utilities|Array|Get(acopy) (Variables|Default|GetGM_NoiseTillingPerStage) (Variables|Default|GetGM_LastStage)))
                (Rendering|Material|SetScalarParameterValueonMaterials _returnvalue "GM_Transition_EmissiveIntensity" (Utilities|Array|Get(aref) (Variables|Default|GetGM_EmissiveIntensityPerStage) (Variables|Default|GetGM_LastStage))))
              (:then_1
                (if (and (> (Math|Integer|Max(Integer) _statvalue _gm_laststage) (Variables|Default|GetGM_StartStage)) (< (Math|Integer|Max(Integer) _statvalue _gm_laststage) _returnvalue_2))
                  (CallFunction|OnActualStatChange))))))))
    (else
      (CallFunction|TransitionLevel 0))))

(event Custom|OnActualStatChange
  ;[1 1 1] GoldMaskGrowingLogic
  ;;[1 1 1] This should only fire off when the stat changes - use this event!
  (Rendering|Material|SetScalarParameterValueonMaterials (CustomCharacter|GetSkeletalMeshForPartOnPlayerPawn) "GM_GameTimeFromBP" (Utilities|Time|GetGameTimeinSeconds)))

(event Custom|InitializeVariabels
  (bind _returnvalue (CustomCharacter|GetSkeletalMeshForPartOnPlayerPawn))
  ; Initialize Variables
  (Rendering|Material|SetScalarParameterValueonMaterials _returnvalue "GM_StartAtStage" (Math|Conversions|ToFloat(Integer) (Variables|Default|GetGM_StartStage)))
  (Rendering|Material|SetScalarParameterValueonMaterials _returnvalue "GM_StopAtStage" (Math|Conversions|ToFloat(Integer) (Variables|Default|GetGM_EndStage)))
  (Rendering|Material|SetScalarParameterValueonMaterials _returnvalue "GM_Transition_EmissiveIntensity" 0.0)
  (Utilities|NotwhenBID)
  (Rendering|Material|SetScalarParameterValueonMaterials _returnvalue "GM_EndRadius" (Utilities|Array|Get(acopy) (Variables|Default|GetGM_EndRadiusPerStage) (Variables|Default|GetGM_LastStage))))

(event EventOnNewWeapons
  (params (WeaponsAsItems :type Array%20of%20Fort%20Item%20Object%20References))
  (bind _pendingitemstowrap (Variables|Default|GetPendingItemsToWrap))
  ;[0.9062 0.5249 0 0.5] Gather Weapons and Apply Midas Touch
  (|CallParentFunction WeaponsAsItems)
  (Utilities|NotwhenBID)
  (for _array_element WeaponsAsItems
    (switch Utilities|FlowControl|Switch|SwitchonEFortItemType (Item|GetType _array_element)
      (:WeaponMelee
        _)
      (:WeaponRanged
        _))
    (Utilities|Array|Add (Variables|Default|GetPendingItemsToWrap) _array_element))
  (if (Utilities|IsValidSoftObjectReference (Variables|Default|GetWeaponWrapToUse))
    (Default|ApplyOverridetoItems _pendingitemstowrap)
    (Utilities|Array|Clear _pendingitemstowrap)))

(event EventOnPlayerPickupDropped
  (params (Pickup :type Fort%20Pickup%20Object%20Reference) (PickupItemEntry :type Fort%20Item%20Entry%20Structure%20%28by%20ref%29))
  ;[1 1 1] Force Weapons And Vehicles To Have A Specific Wrap (MIDAS TOUCH)
  ;;[0.4353 0.4275 0 0.502] Midas Touch Removal on Weapon Dropped
  (Utilities|IsValid Pickup
    (:Is%20Valid
      (if (== (Variables|Default|GetWeaponWrapToUse) (CallFunction|GetPrimaryItemEntryOverrideWrap PickupItemEntry))
        (Class|FortPickup|ClearOverrideWrapInternal Pickup)
        (bind _asfort_player_controller (CallFunction|GetAssociatedPlayerController))
        (Utilities|IsValid _asfort_player_controller
          (:Is%20Valid
            (Class|FortPlayerController|ServerClearOverrideWrap _asfort_player_controller Pickup)))))))

(event EventOnEnterVehicle
  (params (VehicleEntered :type Fort%20Vehicle%20Interface%20Interface%20%28by%20ref%29))
  ;[1 1 1] Force Weapons And Vehicles To Have A Specific Wrap (MIDAS TOUCH)
  (|CallParentFunction VehicleEntered)
  (Utilities|NotwhenBID)
  (if (Utilities|IsValidSoftObjectReference (Variables|Default|GetWeaponWrapToUse))
    (bind _asfort_player_controller (CallFunction|GetAssociatedPlayerController))
    (Utilities|IsValid _asfort_player_controller
      (:Is%20Valid
        (Class|FortPlayerController|ApplyOverrideWraptoVehicle _asfort_player_controller (Variables|Default|GetWeaponWrapToUse))))))

(fn ApplyOverrideToItems
  (params (Array :type Array%20of%20Fort%20Item%20Object%20References))
  (bind _asfort_player_controller (CallFunction|GetAssociatedPlayerController))
  (Utilities|IsValid _asfort_player_controller
    (:Is%20Valid
      (for _array_element Array
        (Class|FortPlayerController|ServerApplyOverrideWraptoItem _asfort_player_controller (Item|GetItemGuid _array_element) (Variables|Default|GetWeaponWrapToUse))))))

(fn CheckIfReactive
  (bind _returnvalue (CustomCharacter|GetAssociatedPlayerPawn))
  (if (Utilities|Enum|Equal(Enum) (CustomCharacter|GetCharacterPartType) "Backpack")
    ;[1 1 1] Backpack
    (bind _index (CosmeticLoadout|GetVariantChannelActiveIndexforSlotActor "(TagName=\"LoadoutSlot.Backpack\")" _returnvalue "(TagName=\"Cosmetics.Variant.Channel.Material\")" (Utilities|Array|MakeArray "(TagName=\"Cosmetics.Variant.Property.Mat1\")" "(TagName=\"Cosmetics.Variant.Property.Mat2\")" "(TagName=\"Cosmetics.Variant.Property.Mat3\")")))
    (switch int _index
      (:0
        (Variables|Default|SetbReactive true))
      (:1
        _)
      (:2
        _))
    (Variables|Default|SetbReactive false)
    (else
      (bind _index_2 (CosmeticLoadout|GetVariantChannelActiveIndexforSlotActor "(TagName=\"LoadoutSlot.Character\")" _returnvalue "(TagName=\"Cosmetics.Variant.Channel.Emissive\")" (Utilities|Array|MakeArray "(TagName=\"Cosmetics.Variant.Property.Emissive1\")" "(TagName=\"Cosmetics.Variant.Property.Emissive2\")")))
      (switch int _index_2
        (:0
          (Variables|Default|SetbReactive true))
        (:1
          (Variables|Default|SetbReactive false))))))

(fn GetAssociatedPlayerController
  (bind _returnvalue (CustomCharacter|GetAssociatedPlayerPawn))
  (Utilities|FlowControl|Sequence
    (:then_0
      (bind _aspawn (Utilities|Casting|CastToPawn (Vehicle|GetVehicleActor _returnvalue :!class /Script/FortniteGame.FortPlayerPawn)))
      (bind _asfort_player_controller (Utilities|Casting|CastToFortPlayerController (Pawn|GetController _aspawn)))
      (return _asfort_player_controller))
    (:then_1
      (bind _asfort_player_controller_2 (Utilities|Casting|CastToFortPlayerController (Pawn|GetController _returnvalue)))
      (return _asfort_player_controller_2))))

(fn TransitionLevel
  (params (Stage :type Integer))
  (bind _returnvalue (CustomCharacter|GetSkeletalMeshForPartOnPlayerPawn))
  (Rendering|Material|SetScalarParameterValueonMaterials _returnvalue "GM_CurrentStage" (Utilities|Array|Get(aref) (Variables|Default|GetGM_ProgessPerStage) Stage))
  (Rendering|Material|SetScalarParameterValueonMaterials _returnvalue "GM_GameTimeFromBP" (Utilities|Time|GetGameTimeinSeconds))
  (Rendering|Material|SetVectorParameterValueonMaterials _returnvalue "GM_TransitionNoise_Tiling" (Utilities|Array|Get(acopy) (Variables|Default|GetGM_NoiseTillingPerStage) Stage))
  (Rendering|Material|SetScalarParameterValueonMaterials _returnvalue "GM_Transition_EmissiveIntensity" (Utilities|Array|Get(aref) (Variables|Default|GetGM_EmissiveIntensityPerStage) Stage)))

(fn ConstructionScript
  (bind _returnvalue (CustomCharacter|GetSkeletalMeshForPartOnPlayerPawn))
  (|CallParentFunction)
  (Rendering|Material|SetScalarParameterValueonMaterials (CustomCharacter|GetSkeletalMeshForPartOnPlayerPawn) "GM_GameTimeFromBP" -99999.0)
  (Rendering|Material|SetScalarParameterValueonMaterials _returnvalue "GM_StartAtStage" (Math|Conversions|ToFloat(Integer) (Variables|Default|GetGM_StartStage)))
  (Rendering|Material|SetScalarParameterValueonMaterials _returnvalue "GM_StopAtStage" (Math|Conversions|ToFloat(Integer) (Variables|Default|GetGM_EndStage)))
  (Rendering|Material|SetScalarParameterValueonMaterials _returnvalue "GM_Transition_EmissiveIntensity" 0.0))
