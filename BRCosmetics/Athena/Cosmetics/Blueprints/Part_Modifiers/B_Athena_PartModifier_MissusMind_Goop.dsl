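; Cosmetic part modifier for the MissusMind Goop character part that drives its glow and body-tracking material effects.
; On post res-in it enables BP_UseCCPM, and adjusts butterfly glow, hand brightness and lightning when the preview and frontend-context checks differ.
; Each tick it feeds hand and chest socket positions, plus finger offset vectors, into the mesh materials.
; Both offsets subtract from the left-hand position, including the right-hand one.
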(event EventOnDayPhaseChanged
  (params (CurrentDayPhase :type EFortDayPhase%20Enum :default "Morning") (PreviousDayPhase :type EFortDayPhase%20Enum :default "Morning") (bAtCreation :type Boolean)))

(event EventOnNewPlayerWeaponType
  (params (NewType :type EFortItemType%20Enum :default "WorldItem") (OldType :type EFortItemType%20Enum :default "WorldItem"))
  (|CallParentFunction NewType OldType) ; This node is disabled and will not be called.\nDrag off pins to build functionality.
)

(event EventOnPostResIn
  (bind _returnvalue (CustomCharacter|GetSkeletalMeshForPartOnPlayerPawn))
  ; SuperLevel
  (|CallParentFunction)
  (Rendering|Material|SetScalarParameterValueonMaterials _returnvalue "BP_UseCCPM" 1.0)
  (if (!= (Reactive|Frontend|IsCosmeticObjectDisplayedasPreview 0) (Reactive|Frontend|IsCosmeticObjectDisplayedinFrontendContext 0))
    (Rendering|Material|SetScalarParameterValueonMaterials _returnvalue "ButterflyGlow_Brightness" 4.0)
    (Rendering|Material|SetScalarParameterValueonMaterials _returnvalue "Butterfly_HandsBrightnessMultiplier" 0.5)
    (Rendering|Material|SetScalarParameterValueonMaterials _returnvalue "LightningIntensity" 12.0)))

(event EventOnWeaponFired)

(event EventOnNewWeapons
  (params (WeaponsAsItems :type Array%20of%20Fort%20Item%20Object%20References)))

(event EventTick
  (params (DeltaSeconds :type Float%20%28single-precision%29))
  (bind _returnvalue (CustomCharacter|GetSkeletalMeshForPartOnPlayerPawn))
  (bind _returnvalue_2 (Transformation|GetSocketLocation _returnvalue "middle_metacarpal_l"))
  ; SuperLevel
  (Rendering|Material|SetVectorParameterValueonMaterials _returnvalue "BP_HandPosition_L" _returnvalue_2)
  (Rendering|Material|SetVectorParameterValueonMaterials _returnvalue "BP_HandPosition_R" (Transformation|GetSocketLocation _returnvalue "middle_metacarpal_r"))
  (Rendering|Material|SetVectorParameterValueonMaterials _returnvalue "BP_ChestPosition" (Transformation|GetSocketLocation _returnvalue "spine_05"))
  (Rendering|Material|SetVectorParameterValueonMaterials _returnvalue "BP_OffsetVector_L" (- _returnvalue_2 (Transformation|GetSocketLocation _returnvalue "middle_01_l")))
  (Rendering|Material|SetVectorParameterValueonMaterials _returnvalue "BP_OffsetVector_R" (- _returnvalue_2 (Transformation|GetSocketLocation _returnvalue "middle_01_r"))))

(event EventOnCurieStateAttached
  (params (Tag :type Gameplay%20Tag%20Structure%20%28by%20ref%29)))

(event EventOnCurieStateDetached
  (params (Tag :type Gameplay%20Tag%20Structure%20%28by%20ref%29)))

(event EventOnBeginSkydiving
  (|CallParentFunction) ; This node is disabled and will not be called.\nDrag off pins to build functionality.
)

(event EventOnEndSkydiving
  (|CallParentFunction) ; This node is disabled and will not be called.\nDrag off pins to build functionality.
)

(event EventOnEnterVehicle
  (params (VehicleEntered :type Fort%20Vehicle%20Interface%20Interface%20%28by%20ref%29))
  (|CallParentFunction VehicleEntered) ; This node is disabled and will not be called.\nDrag off pins to build functionality.
)

(event EventOnExitVehicle
  (|CallParentFunction) ; This node is disabled and will not be called.\nDrag off pins to build functionality.
)

(event EventOnTargetingOrFirstPersonCameraChanged)

(event EventOnHitPawn
  (params (SelfActor :type Actor%20Object%20Reference) (OtherActor :type Actor%20Object%20Reference) (NormalImpulse :type Vector :default "0, 0, 0") (Hit :type Hit%20Result%20Structure%20%28by%20ref%29)))

(fn ConstructionScript
  (|CallParentFunction))
