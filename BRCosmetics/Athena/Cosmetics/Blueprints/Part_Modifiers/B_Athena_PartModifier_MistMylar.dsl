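; Cosmetic part modifier for MistMylar that blanks some cape materials when the player pawn appears as a front-end hologram.
; After res-in, if in the front end, it starts a looping timer that polls the pawn's hologram flag several times a second.
; When flagged, it checks the part's skeletal mesh against the two MistMylar cape meshes and overrides material slots 1–3 with M_Blank.
; Nothing here clears the timer or restores the materials.
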(event EventOnPostResIn
  (|CallParentFunction)
  (if (CustomCosmeticModifier|IsinFrontEnd)
    (Utilities|Time|SetTimerbyEvent (delegate Custom|CE_FECheck) 0.2 true)))

(event Custom|CE_FECheck
  (bind _returnvalue (CustomCharacter|GetSkeletalMeshForPartOnPlayerPawn))
  (if (Variables|/Script/FortniteGame.FortPlayerPawn|GetbIsInFrontEndHologram (CustomCharacter|GetAssociatedPlayerPawn))
    (bind _asskeletal_mesh_component (Utilities|Casting|CastToSkeletalMeshComponent _returnvalue))
    (if (or (== (Variables|/Script/Engine.SkinnedMeshComponent|GetSkeletalMesh _asskeletal_mesh_component) "/BRCosmetics/Accessories/FORT_Capes/Cape_M_MED_MistMylar/Meshes/M_MED_MistMylar_Cape.M_MED_MistMylar_Cape") (== (Variables|/Script/Engine.SkinnedMeshComponent|GetSkeletalMesh _asskeletal_mesh_component) "/BRCosmetics/Accessories/FORT_Capes/Cape_M_MED_MistMylar/Meshes/M_MED_MistMylarUniversal_Cape.M_MED_MistMylarUniversal_Cape"))
      (Rendering|Material|SetMaterial _returnvalue 1 "/Game/Accessories/FORT_Backpacks/M_MED_Celestial_Backpack/M_Blank.M_Blank")
      (Rendering|Material|SetMaterial _returnvalue 2 "/Game/Accessories/FORT_Backpacks/M_MED_Celestial_Backpack/M_Blank.M_Blank")
      (Rendering|Material|SetMaterial _returnvalue 3 "/Game/Accessories/FORT_Backpacks/M_MED_Celestial_Backpack/M_Blank.M_Blank"))))

(fn ConstructionScript
  (|CallParentFunction))
