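; Cosmetic part modifier for Mistress Sombre that makes the character part react to the player's eliminations.
; After res-in, it watches the kill stat and stores a Reactivity Enabled flag from the backpack's Parts variant lookup, though this graph never reads it.
; Each stat change above zero stamps game time into the part's dynamic material.
; A DoOnce limits the idle VFX reactive pulse to a single three-second burst.
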(event EventOnPostResIn
  (|CallParentFunction)
  (Utilities|FlowControl|Sequence
    (:then_0
      ; Register Reactivity On Kills
      (bind (_statname _statvalue) (Cosmetics|WatchCosmeticStat (CustomCharacter|GetAssociatedPlayerPawn) "(TagName=\"StatManager.AthenaKills\")")
        (:OnStatChanged
          (if (> _statvalue 0)
            (CallFunction|BallReactivity)))))
    (:then_1
      ; Get CIDs
      (bind (_index _bindexisvalid) (Class|FortCosmeticFunctionLibrary|GetVariantChannelActiveIndexForItem "/BRCosmetics/Athena/Items/Cosmetics/Backpacks/Backpack_MistressSombre.Backpack_MistressSombre" (CosmeticLoadout|GetCosmeticLoadout) "(TagName=\"Cosmetics.Variant.Channel.Parts\")" (Utilities|Array|MakeArray "(TagName=\"Cosmetics.Variant.Property.Stage1\")") false false))
      (Variables|Default|SetReactivity%20Enabled _bindexisvalid))))

(event Custom|Ball_Reactivity
  (bind _returnvalue (CallFunction|GetIdleVFXComponent))
  ; Reactivity Event
  (Rendering|Material|SetScalarParameterValue (Class|BTechArtGlobalLibrary|GetDynamicMaterialInstance (CustomCharacter|GetSkeletalMeshForPartOnPlayerPawn)) "Timestamp" (Utilities|Time|GetGameTimeinSeconds) :!class /Script/Engine.MaterialInstanceDynamic) ; Enable material reactivity
  ; FX Reactivity
  (Utilities|FlowControl|DoOnce false
    (:Completed
      (Effects|Components|ParticleSystem|SetBoolParameter _returnvalue "FX_Reactive" true)
      (Utilities|FlowControl|Delay 3.0)
      (Effects|Components|ParticleSystem|SetBoolParameter _returnvalue "FX_Reactive" false))))

(fn ConstructionScript
  (|CallParentFunction))
