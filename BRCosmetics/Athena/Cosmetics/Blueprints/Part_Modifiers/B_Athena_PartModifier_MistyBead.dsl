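; Character part modifier for the MistyBead cosmetic that makes its materials and VFX react to eliminations.
; A positive stat change writes current and previous kill counts plus timing to mesh material parameters for shader-driven transitions.
; It also replaces the previous elimination Niagara effect, passing kill count, frontend and hidden-body-part flags.
; In the frontend, post-res-in logic simulates reactivity with delays and a kill counter that resets at three.
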(event EventOnPostResIn
  (bind _fekillcount (Variables|Default|GetFEKillCount))
  (|CallParentFunction)
  (if (CallFunction|ShouldDisplayFrontendReactivity)
    (Utilities|FlowControl|Delay 2.0)
    (if (>= (Variables|Default|GetFEKillCount) 3)
      (bind _output_get (Variables|Default|SetFEKillCount 0))
      (CallFunction|OnReactivity _output_get 4.0)
      (bind _returnvalue (CallFunction|GetVFXSpawnParameters "(TagName=\"CosmeticPart.VFX.Elimination\")")
        (:Success
          (bind _returnvalue_2 (PartModifier|SpawnSystem|SpawnSystemAttachedPartModifier (Utilities|Struct|BreakCharacterPartModifierVFXSpawnParameters _returnvalue) (.AttachToComponent (Utilities|Struct|BreakCharacterPartModifierVFXSpawnParameters _returnvalue)) (.SocketName (Utilities|Struct|BreakCharacterPartModifierVFXSpawnParameters _returnvalue)) "0, 0, 0" "0, 0, 0" (.LocationType (Utilities|Struct|BreakCharacterPartModifierVFXSpawnParameters _returnvalue)) true (.bAutoActivate (Utilities|Struct|BreakCharacterPartModifierVFXSpawnParameters _returnvalue)))
            (:Success
              (bind _output_get_2 (Variables|Default|SetEliminationParticle _returnvalue_2))
              (Niagara|SetNiagaraVariable(Bool) _output_get_2 "bisFrontEnd" (Reactive|Frontend|IsCosmeticObjectDisplayedinFrontendContext 0))
              (Utilities|FlowControl|Delay 8.0))))
        (:Failure
          _))
      (Effects|Components|ParticleSystem|SetIntParameter (CallFunction|GetVFXComponent "(TagName=\"CosmeticPart.VFX.Idle\")") "KillCount" _fekillcount)
      (Utilities|FlowControl|Delay 8.0))
    (Math|Integer|IncrementInt (Variables|Default|GetFEKillCount)
      (:%20%20
        (CallFunction|OnReactivity _fekillcount)))))

(event EventOnStatChanged
  (params (State%20Name :type Name :default "None") (Stat%20Value :type Integer))
  (|CallParentFunction State%20Name Stat%20Value)
  (if (> Stat%20Value 0)
    (CallFunction|OnReactivity Stat%20Value)
    (bind _returnvalue (CallFunction|GetVFXSpawnParameters "(TagName=\"CosmeticPart.VFX.Elimination\")")
      (:Success
        ; Disables a particle to avoid stacking after getting multiple elims (Set the NS to Kill in order to make this works
        (Components|Activation|Deactivate (Variables|Default|GetEliminationParticle))
        (bind _returnvalue_2 (PartModifier|SpawnSystem|SpawnSystemAttachedPartModifier (Utilities|Struct|BreakCharacterPartModifierVFXSpawnParameters _returnvalue) (.AttachToComponent (Utilities|Struct|BreakCharacterPartModifierVFXSpawnParameters _returnvalue)) (.SocketName (Utilities|Struct|BreakCharacterPartModifierVFXSpawnParameters _returnvalue)) "0, 0, 0" "0, 0, 0" (.LocationType (Utilities|Struct|BreakCharacterPartModifierVFXSpawnParameters _returnvalue)) true (.bAutoActivate (Utilities|Struct|BreakCharacterPartModifierVFXSpawnParameters _returnvalue)))
          (:Success
            (bind _output_get (Variables|Default|SetEliminationParticle _returnvalue_2))
            (Niagara|SetNiagaraVariable(Bool) _output_get "bisFrontEnd" (Reactive|Frontend|IsCosmeticObjectDisplayedinFrontendContext 0))
            (Niagara|SetNiagaraVariable(Int32) _output_get "KillCount" Stat%20Value)
            (Default|SetSpawnedParticleHiddenTags _returnvalue_2))))
      (:Failure
        (Effects|Components|ParticleSystem|SetIntParameter (CallFunction|GetVFXComponent "(TagName=\"CosmeticPart.VFX.Idle\")") "KillCount" Stat%20Value)))))

(event Custom|OnReactivity
  (params (KillCount :type Integer) (EffectLength :type Float%20%28double-precision%29 :default 2.0))
  (bind _returnvalue (CustomCharacter|GetSkeletalMeshforPartonOwner))
  (Rendering|Material|SetScalarParameterValueonMaterials _returnvalue "PreviousKillCount" (Math|Conversions|ToFloat(Integer) (Variables|Default|GetFEPreviousKillCount)))
  (Rendering|Material|SetScalarParameterValueonMaterials _returnvalue "KillCount" (Math|Conversions|ToFloat(Integer) KillCount))
  (Rendering|Material|SetScalarParameterValueonMaterials _returnvalue "TimestampLength" EffectLength)
  (Rendering|Material|SetScalarParameterValueonMaterials _returnvalue "Timestamp" (Utilities|Time|GetGameTimeinSeconds))
  (Variables|Default|SetFEPreviousKillCount KillCount))

(fn SetSpawnedParticleHiddenTags
  (params (NiagaraParticle :type Niagara%20Particle%20System%20Component%20Object%20Reference))
  (Niagara|SetNiagaraVariable(Bool) NiagaraParticle "bisHeadHidden" (Variables|HiddenPartCheck|GetbIsHeadHidden))
  (Niagara|SetNiagaraVariable(Bool) NiagaraParticle "bisHandHidden" (Variables|HiddenPartCheck|GetbIsHandHidden))
  (Niagara|SetNiagaraVariable(Bool) NiagaraParticle "bisFootHidden" (Variables|HiddenPartCheck|GetbIsFootHidden)))

(fn ConstructionScript
  (|CallParentFunction))
