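; Character part modifier for the Mochi skin that makes the face smile on gameplay stats and plays a burst effect when the pawn lands.
; On BeginPlay and post-res-in it resets SmileOverride, then watches kills, damage received, Royal Royales and chest opens.
; The smile is a material scalar held by a four-second retriggerable delay.
; The cached Material variant index selects the normal or dark Niagara landing burst at FX_Burst1.
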(event EventBeginPlay
  (bind _returnvalue (CustomCharacter|GetAssociatedPlayerPawn))
  (bind _returnvalue_2 (CustomCharacter|GetSkeletalMeshForPartOnPlayerPawn))
  (Development|PrintString "TEST2")
  (Utilities|FlowControl|Sequence
    (:then_0
      (Rendering|Material|SetScalarParameterValueonMaterials (CustomCharacter|GetSkeletalMeshForPartOnPlayerPawn) "SmileOverride" 0.0) ; Make sure the smile override is always 0 at the begining of the match
      ;[1 1 1] Watch all reactive Stats and Smile for 5 seconds
      (Utilities|FlowControl|Sequence
        (:then_0
          (Cosmetics|WatchCosmeticStat _returnvalue "(TagName=\"StatManager.AthenaKills\")" 0 false
            (:OnStatChanged
              _)))
        (:then_1
          (Cosmetics|WatchCosmeticStat _returnvalue "(TagName=\"StatManager.DamageReceived\")" 0 false
            (:OnStatChanged
              _)))
        (:then_2
          (Cosmetics|WatchCosmeticStat _returnvalue "(TagName=\"StatManager.TotalRoyalRoyales\")" 0 false
            (:OnStatChanged
              _)))
        (:then_3
          (Cosmetics|WatchCosmeticStat _returnvalue "(TagName=\"StatManager.OpenChests\")" 0 false
            (:OnStatChanged
              _))))
      (Rendering|Material|SetScalarParameterValueonMaterials _returnvalue_2 "SmileOverride" 1.0)
      (Utilities|FlowControl|RetriggerableDelay 4.0)
      (Rendering|Material|SetScalarParameterValueonMaterials _returnvalue_2 "SmileOverride" 0.0))
    (:then_1
      (bind _index (CosmeticLoadout|GetVariantChannelActiveIndexforSlotActor "(TagName=\"LoadoutSlot.Character\")" (CustomCharacter|GetAssociatedPlayerPawn) "(TagName=\"Cosmetics.Variant.Channel.Material\")" (Utilities|Array|MakeArray "(TagName=\"Cosmetics.Variant.Property.Mat1\")" "(TagName=\"Cosmetics.Variant.Property.Mat2\")")))
      (Variables|Default|SetVariantIndex _index))))

(event EventOnPostResIn
  (bind _returnvalue (CustomCharacter|GetAssociatedPlayerPawn))
  (bind _returnvalue_2 (CustomCharacter|GetSkeletalMeshForPartOnPlayerPawn))
  (|CallParentFunction)
  (Utilities|FlowControl|Sequence
    (:then_0
      (Rendering|Material|SetScalarParameterValueonMaterials (CustomCharacter|GetSkeletalMeshForPartOnPlayerPawn) "SmileOverride" 0.0) ; Make sure the smile override is always 0 at the begining of the match
      ;[1 1 1] Watch all reactive Stats and Smile for 5 seconds
      (Utilities|FlowControl|Sequence
        (:then_0
          (Cosmetics|WatchCosmeticStat _returnvalue "(TagName=\"StatManager.AthenaKills\")" 0 false
            (:OnStatChanged
              _)))
        (:then_1
          (Cosmetics|WatchCosmeticStat _returnvalue "(TagName=\"StatManager.DamageReceived\")" 0 false
            (:OnStatChanged
              _)))
        (:then_2
          (Cosmetics|WatchCosmeticStat _returnvalue "(TagName=\"StatManager.TotalRoyalRoyales\")" 0 false
            (:OnStatChanged
              _)))
        (:then_3
          (Cosmetics|WatchCosmeticStat _returnvalue "(TagName=\"StatManager.OpenChests\")" 0 false
            (:OnStatChanged
              _))))
      (Rendering|Material|SetScalarParameterValueonMaterials _returnvalue_2 "SmileOverride" 1.0)
      (Utilities|FlowControl|RetriggerableDelay 4.0)
      (Rendering|Material|SetScalarParameterValueonMaterials _returnvalue_2 "SmileOverride" 0.0))
    (:then_1
      (bind _index (CosmeticLoadout|GetVariantChannelActiveIndexforSlotActor "(TagName=\"LoadoutSlot.Character\")" (CustomCharacter|GetAssociatedPlayerPawn) "(TagName=\"Cosmetics.Variant.Channel.Material\")" (Utilities|Array|MakeArray "(TagName=\"Cosmetics.Variant.Property.Mat1\")" "(TagName=\"Cosmetics.Variant.Property.Mat2\")")))
      (Variables|Default|SetVariantIndex _index))))

(event EventOnPawnLanded
  (|CallParentFunction)
  (PartModifier|SpawnSystem|SpawnSystemAttachedPartModifier (Utilities|Select :Index (Variables|Default|GetVariantIndex) :Option%200 "/BRCosmetics/Characters/Player/Female/Medium/Bodies/F_MED_Mochi/FX/NS_Mochi_Skin_Burst.NS_Mochi_Skin_Burst" :Option%201 "/BRCosmetics/Characters/Player/Female/Medium/Bodies/F_MED_Mochi/FX/NS_Mochi_Skin_BurstDark.NS_Mochi_Skin_BurstDark") (CustomCharacter|GetSkeletalMeshForPartOnPlayerPawn) "FX_Burst1"))

(fn ConstructionScript
  (|CallParentFunction))
