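; Cosmetic part modifier for NebulaBurden that keeps a material-driven orb effect aligned to the character's spine.
; It moves its tick to post-update work, then each tick reads the world transform of spine_05 on the part's skeletal mesh.
; It writes that location, plus the bone's X and Y axes as up and forward, into the mesh's dynamic material.
; It also sets a scalar flag enabling this supplied orb location.
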(event EventBeginPlay
  (|CallParentFunction)
  (Actor|Tick|SetTickGroup "TG_PostUpdateWork"))

(event EventTick
  (params (DeltaSeconds :type Float%20%28single-precision%29))
  (bind _returnvalue (CustomCharacter|GetSkeletalMeshForPartOnPlayerPawn))
  (bind _returnvalue_2 (Transformation|GetSocketTransform _returnvalue "spine_05" "RTS_World"))
  (bind _returnvalue_3 (Math|Transform|ToMatrix(Transform) _returnvalue_2))
  (bind (_x _y) (Math|Matrix|GetUnitAxes(Matrix) _returnvalue_3))
  (bind _material_instance_dynamic (Class|BTechArtGlobalLibrary|GetDynamicMaterialInstance _returnvalue))
  (|CallParentFunction DeltaSeconds)
  (Rendering|Material|SetVectorParameterValue _material_instance_dynamic "OrbLocation" (Math|Conversions|ToLinearColor(Vector) (.location _returnvalue_2)) :!class /Script/Engine.MaterialInstanceDynamic)
  (Rendering|Material|SetVectorParameterValue _material_instance_dynamic "OrbUpAxis" (Math|Conversions|ToLinearColor(Vector) _x) :!class /Script/Engine.MaterialInstanceDynamic)
  (Rendering|Material|SetVectorParameterValue _material_instance_dynamic "OrbForwardAxis" (Math|Conversions|ToLinearColor(Vector) _y) :!class /Script/Engine.MaterialInstanceDynamic)
  (Rendering|Material|SetScalarParameterValue _material_instance_dynamic "UseCCPMOrbLocation" 1.0 :!class /Script/Engine.MaterialInstanceDynamic))

(event EventOnPostResIn
  (|CallParentFunction))

(fn ConstructionScript)
