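; Neon Animals cosmetic part modifier that switches the cosmetic's variant when the player changes weapon type, with swap effects.
; Weapon types map to variant tags, applied only when the tag changes.
; An optional cooldown timer stores changes arriving meanwhile and re-evaluates on expiry.
; Each switch spawns an attached particle burst tinted orange for ranged or purple for melee and harvest, plus a matching sound at the part.
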(event EventOnNewPlayerWeaponType
  (params (NewType :type EFortItemType%20Enum :default "WorldItem") (OldType :type EFortItemType%20Enum :default "WorldItem"))
  (if (Utilities|Time|IsTimerActivebyHandle (Variables|PartModifierControls|WeaponControlledVariantSwitcher|GetWeaponVariantSwitcherDelay))
    (Variables|PartModifierControls|WeaponControlledVariantSwitcher|SetDelayed%20Weapon%20Type NewType)
    (else
      (bind _data (Utilities|FlowControl|MergeExecutionwithData NewType (Variables|PartModifierControls|WeaponControlledVariantSwitcher|GetDelayed%20Weapon%20Type))
        (:Then
          (CallFunction|HandlePotentialNewWeaponType _data))))))

(event Custom|OnDelayExpired
  (bind _data (Utilities|FlowControl|MergeExecutionwithData (stale EventOnNewPlayerWeaponType NewType) (Variables|PartModifierControls|WeaponControlledVariantSwitcher|GetDelayed%20Weapon%20Type))
    (:Then
      (CallFunction|HandlePotentialNewWeaponType _data))))

(event Custom|ActuallyChangedWeaponType
  (params (New%20Type :type EFortItemType%20Enum%20%28by%20ref%29 :default "WorldItem"))
  (switch Utilities|FlowControl|Switch|SwitchonEFortItemType New%20Type
    (:WeaponMelee
      (Variables|Default|SetColorLerp 1.0 :!label _label))
    (:WeaponRanged
      (Variables|Default|SetColorLerp 0.0))
    (:WeaponHarvest
      (goto _label)))
  (bind _returnvalue (PartModifier|SpawnSystem|SpawnEmitterAttachedPartModifier(Legacy) "/Game/Effects/Fort_Effects/Effects/Characters/Athena_Parts/Neon/P_NeonAnimals_Swap.P_NeonAnimals_Swap" (CustomCharacter|GetSkeletalMeshForPartOnPlayerPawn) "root")
    (:Success
      (Effects|Components|ParticleSystem|SetVectorParameter _returnvalue "InstanceColorScaleOverLife" (Math|Conversions|ToVector(LinearColor) (Math|Color|Lerp(LinearColor) "(R=1.000000,G=0.176456,B=0.000000,A=1.000000)" "(R=0.332028,G=0.000000,B=0.687500,A=1.000000)" (Variables|Default|GetColorLerp)))))
    (:Failure
      _))
  (Audio|PlaySoundatLocation (select (== (Variables|Default|GetColorLerp) 0.0) "/BRCosmetics/Sounds/Backblings/SkinTransition/Quest_NeonAnimal_Swap_Weapon_Cue.Quest_NeonAnimal_Swap_Weapon_Cue" "/BRCosmetics/Sounds/Backblings/SkinTransition/Quest_NeonAnimal_Swap_Cue.Quest_NeonAnimal_Swap_Cue") (Transformation|GetWorldLocation (Customization|GetSkeletalMeshforPartType (CustomCharacter|GetAssociatedPlayerPawn) (CustomCharacter|GetCharacterPartType) :!class /Script/FortniteGame.FortPlayerPawn))))

(fn HandlePotentialNewWeaponType
  (params (Type :type EFortItemType%20Enum :default "WorldItem"))
  (bind _type (Variables|GetType))
  (bind _forcedvariantsbyitemtype (Variables|PartModifierControls|WeaponControlledVariantSwitcher|GetForcedVariantsByItemType))
  (bind (_value _returnvalue) (Utilities|Map|Find _forcedvariantsbyitemtype _type))
  (Variables|PartModifierControls|WeaponControlledVariantSwitcher|SetDelayed%20Weapon%20Type Type)
  (if _returnvalue
    (Variables|PartModifierControls|WeaponControlledVariantSwitcher|SetType%20Tag _value)
    (if (not (GameplayTags|Equal(GameplayTag) (Variables|PartModifierControls|WeaponControlledVariantSwitcher|GetLast%20Tag) (Variables|PartModifierControls|WeaponControlledVariantSwitcher|GetType%20Tag)))
      (CustomCharacter|SetCosmeticActiveVariant (Variables|PartModifierControls|WeaponControlledVariantSwitcher|GetTarget%20Cosmetic) (Variables|PartModifierControls|WeaponControlledVariantSwitcher|GetChannel%20Tag) (Variables|PartModifierControls|WeaponControlledVariantSwitcher|GetType%20Tag))
      (CallFunction|ActuallyChangedWeaponType (Variables|GetType))
      (Variables|PartModifierControls|WeaponControlledVariantSwitcher|SetLast%20Tag (Variables|PartModifierControls|WeaponControlledVariantSwitcher|GetType%20Tag))
      (if (> (Variables|PartModifierControls|WeaponControlledVariantSwitcher|GetDelay%20Between%20Activations) 0.0)
        (bind _returnvalue_2 (Utilities|Time|SetTimerbyEvent (EventDispatchers|CreateEvent :!event OnDelayExpired) (Variables|PartModifierControls|WeaponControlledVariantSwitcher|GetDelay%20Between%20Activations)))
        (Variables|PartModifierControls|WeaponControlledVariantSwitcher|SetWeaponVariantSwitcherDelay _returnvalue_2)))))

(fn ConstructionScript)
