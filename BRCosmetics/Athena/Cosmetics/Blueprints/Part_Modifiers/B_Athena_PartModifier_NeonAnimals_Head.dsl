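; Character part modifier for the Neon Animals head cosmetic that colors a head light effect according to the held weapon type.
; On weapon type change, a melee or ranged type sets ColorLerp to choose between orange and purple.
; It spawns the head light emitter attached at the part mesh's root.
; On success, it feeds the blended color into the particle's color-scale and light-color vector parameters.
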(event EventActuallyChangedWeaponType
  (params (New%20Type :type EFortItemType%20Enum%20%28by%20ref%29 :default "WorldItem"))
  (bind _colorlerp (Variables|Default|GetColorLerp))
  (bind _returnvalue (Math|Color|Lerp(LinearColor) "(R=1.000000,G=0.176456,B=0.000000,A=1.000000)" "(R=0.332028,G=0.000000,B=0.687500,A=1.000000)" _colorlerp))
  (bind _returnvalue_2 (Math|Conversions|ToVector(LinearColor) _returnvalue))
  (switch Utilities|FlowControl|Switch|SwitchonEFortItemType New%20Type
    (:WeaponMelee
      (Variables|Default|SetColorLerp 0.0))
    (:WeaponRanged
      (Variables|Default|SetColorLerp 1.0))
    (:WeaponHarvest
      _))
  (bind _returnvalue_3 (PartModifier|SpawnSystem|SpawnEmitterAttachedPartModifier(Legacy) "/Game/Effects/Fort_Effects/Effects/Characters/Athena_Parts/Neon/P_NeonAnimals_HeadLight.P_NeonAnimals_HeadLight" (CustomCharacter|GetSkeletalMeshForPartOnPlayerPawn) "root")
    (:Success
      (Effects|Components|ParticleSystem|SetVectorParameter _returnvalue_3 "InstanceColorScaleOverLife" _returnvalue_2)
      (Effects|Components|ParticleSystem|SetVectorParameter _returnvalue_3 "ColorOfLight" _returnvalue_2))))

(fn ConstructionScript)
