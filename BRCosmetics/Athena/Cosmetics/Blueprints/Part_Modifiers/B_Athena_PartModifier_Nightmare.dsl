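; Cosmetic part modifier for the Nightmare outfit that controls the mask's laser material glow and idle particle effect.
; On post-init it derives MaskEquipped from the Parts variant; mask on/off anim notifies also update it.
; It creates a face-accessory dynamic material, attached outside front-end holograms, with emissive raised in the front end and zeroed elsewhere.
; Emote and camera-change events use a timeline and DoOnce gates to fade emissive and particle alpha together.
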(event Custom|AnimNotify_MaskOff
  (Variables|Default|SetMaskEquipped false)
  (if (CustomCosmeticModifier|IsinFrontEnd)
    (CallFunction|OnMaskOFFEmoteFE)))

(event Custom|AnimNotify_MaskOn
  (bind _lasermid (Variables|Default|GetLaserMID))
  (Variables|Default|SetMaskEquipped true)
  (if (CustomCosmeticModifier|IsinFrontEnd)
    (Utilities|FlowControl|Delay 4.0)
    (bind _returnvalue (Rendering|Material|CreateDynamicMaterialInstance "/BRCosmetics/Characters/Player/Male/Medium/Bodies/M_MED_Nightmare/Materials/M_MED_Nightmare_FaceAcc.M_MED_Nightmare_FaceAcc" "LaserControlMID" "None" :!class /Script/Engine.KismetMaterialLibrary))
    (Variables|Default|SetLaserMID _returnvalue)
    (CallFunction|OnMaskONEmoteFE)
    (else
      (bind _returnvalue_2 (Rendering|Material|CreateDynamicMaterialInstance "/BRCosmetics/Characters/Player/Male/Medium/Bodies/M_MED_Nightmare/Materials/M_MED_Nightmare_FaceAcc.M_MED_Nightmare_FaceAcc" "LaserControlMID" "None" :!class /Script/Engine.KismetMaterialLibrary))
      (Variables|Default|SetLaserMID _returnvalue_2)
      (if (CustomCosmeticModifier|IsinFrontEnd)
        (Rendering|Material|SetScalarParameterValue _lasermid "emissive mult" 1e+01 :!class /Script/Engine.MaterialInstanceDynamic)
        (Utilities|FlowControl|Delay 1.0)
        (Effects|Components|ParticleSystem|SetFloatParameter (CallFunction|GetIdleVFXComponent) "MasterAlpha" 1.0)
        (else
          (Rendering|Material|SetScalarParameterValue _lasermid "emissive mult" 0.0 :!class /Script/Engine.MaterialInstanceDynamic)
          (Utilities|FlowControl|Delay 1.0)))
      (if (not (Variables|/Script/FortniteGame.FortPlayerPawn|GetbIsInFrontEndHologram (CustomCharacter|GetAssociatedPlayerPawn)))
        (Material|ReattachOrCreateMID (CustomCharacter|GetSkeletalMeshForPartOnPlayerPawn) (Variables|Default|GetLaserMID))))))

(event EventOnPostResIn
  (bind _lasermid (Variables|Default|GetLaserMID))
  (|CallParentFunction)
  (Effects|Components|ParticleSystem|SetBoolParameter (CallFunction|GetIdleVFXComponent) "bIsPlayersPawn" (HelperFunctions|IsLocalPawn (CustomCharacter|GetAssociatedPlayerPawn)))
  (bind (_index _bindexisvalid _bfoundchannel) (Class|FortCosmeticFunctionLibrary|GetVariantChannelActiveIndexForSlot "(TagName=\"LoadoutSlot.Character\")" (CosmeticLoadout|GetCosmeticLoadout) "(TagName=\"Cosmetics.Variant.Channel.Parts\")" (Utilities|Array|MakeArray "(TagName=\"Cosmetics.Variant.Property.Stage1\")" "(TagName=\"Cosmetics.Variant.Property.Stage2\")") false false))
  (bind _output_get (Variables|Default|SetMaskEquipped (or (== _index 0) (not _bfoundchannel))))
  (if _output_get
    (bind _returnvalue (Rendering|Material|CreateDynamicMaterialInstance "/BRCosmetics/Characters/Player/Male/Medium/Bodies/M_MED_Nightmare/Materials/M_MED_Nightmare_FaceAcc.M_MED_Nightmare_FaceAcc" "LaserControlMID" "None" :!class /Script/Engine.KismetMaterialLibrary))
    (Variables|Default|SetLaserMID _returnvalue)
    (if (CustomCosmeticModifier|IsinFrontEnd)
      (Rendering|Material|SetScalarParameterValue _lasermid "emissive mult" 1e+01 :!class /Script/Engine.MaterialInstanceDynamic)
      (Utilities|FlowControl|Delay 1.0)
      (Effects|Components|ParticleSystem|SetFloatParameter (CallFunction|GetIdleVFXComponent) "MasterAlpha" 1.0)
      (else
        (Rendering|Material|SetScalarParameterValue _lasermid "emissive mult" 0.0 :!class /Script/Engine.MaterialInstanceDynamic)
        (Utilities|FlowControl|Delay 1.0)))
    (if (not (Variables|/Script/FortniteGame.FortPlayerPawn|GetbIsInFrontEndHologram (CustomCharacter|GetAssociatedPlayerPawn)))
      (Material|ReattachOrCreateMID (CustomCharacter|GetSkeletalMeshForPartOnPlayerPawn) (Variables|Default|GetLaserMID)))))

(event Custom|OnMaskON_Emote_FE
  (bind _returnvalue (CallFunction|GetIdleVFXComponent))
  (Utilities|FlowControl|DoOnce false
    (:Completed
      (Components|Activation|Activate _returnvalue false)
      (bind (_direction _newtrack_0) (Timeline|Timeline_0 PlayFromStart)
        (:Update
          (Effects|Components|ParticleSystem|SetFloatParameter _returnvalue "MasterAlpha" _newtrack_0)
          (Rendering|Material|SetScalarParameterValue (Variables|Default|GetLaserMID) "emissive mult" (* _newtrack_0 1e+01) :!class /Script/Engine.MaterialInstanceDynamic))
        (:Finished
          (Utilities|FlowControl|DoOnce false
            (:Completed
              (Effects|Components|ParticleSystem|SetFloatParameter _returnvalue "MasterAlpha" 0.0)
              (Rendering|Material|SetScalarParameterValue (Variables|Default|GetLaserMID) "emissive mult" 0.0 :!class /Script/Engine.MaterialInstanceDynamic)
              (Components|Activation|Deactivate _returnvalue))))))))

(event Custom|OnMaskOFF_Emote_FE
  (bind _returnvalue (CallFunction|GetIdleVFXComponent))
  (Utilities|FlowControl|Sequence
    (:then_0
      (Timeline|Timeline_0 ReverseFromEnd))
    (:then_1
      (Utilities|FlowControl|DoOnce false
        (:Completed
          (Components|Activation|Activate _returnvalue false)
          (Timeline|Timeline_0 PlayFromStart))))))

(event EventOnTargetingOrFirstPersonCameraChanged
  (bind _returnvalue (CallFunction|GetIdleVFXComponent))
  (|CallParentFunction)
  (if (not (Variables|Default|GetMaskEquipped))
    (Utilities|FlowControl|DoOnce false
      (:Completed
        (Effects|Components|ParticleSystem|SetFloatParameter _returnvalue "MasterAlpha" 0.0)
        (Rendering|Material|SetScalarParameterValue (Variables|Default|GetLaserMID) "emissive mult" 0.0 :!class /Script/Engine.MaterialInstanceDynamic)
        (Components|Activation|Deactivate _returnvalue))))
  (Utilities|FlowControl|DoOnce false
    (:Completed
      (Components|Activation|Activate _returnvalue false)
      (Timeline|Timeline_0 PlayFromStart))))

(fn ConstructionScript
  (|CallParentFunction))
