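; Cosmetic character part modifier for Nosh that toggles its idle VFX, apparently a drip effect, around skydiving.
; When skydiving begins, it deactivates the component returned by GetIdleVFXComponent.
; When skydiving ends, it activates that component again with reset enabled.
; The construction script is empty, and the file shows no other logic.
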(event EventOnBeginSkydiving
  (Components|Activation|Deactivate (CallFunction|GetIdleVFXComponent)))

(event EventOnEndSkydiving
  (Components|Activation|Activate (CallFunction|GetIdleVFXComponent) true))

(fn ConstructionScript)
