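; Part modifier Blueprint for the Octopus cosmetic character part that adds a particle effect when the pawn lands.
; On the landed event it calls the parent, then spawns the Coins Jump emitter through the legacy part-modifier spawn function.
; The emitter snaps to the dyn_Pipe point on the part's skeletal mesh, auto-destroys, matches pawn custom depth and has debounce enabled.
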(event EventOnPawnLanded
  (|CallParentFunction)
  (PartModifier|SpawnSystem|SpawnEmitterAttachedPartModifier(Legacy) :EmitterTemplate (Variables|Default|GetCoins%20Jump) :AttachToComponent (CustomCharacter|GetSkeletalMeshForPartOnPlayerPawn) :AttachPointName "dyn_Pipe" :Scale "1.000000,1.000000,1.000000" :LocationType "SnapToTarget" :bAutoDestroy true :PoolingMethod "None" :bAutoActivate true :bMatchPawnCustomDepth true :bDebounce true :DebounceTime -1.0))

(fn ConstructionScript)
