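; Part modifier for the male Oktoberfest back bling that plays a cuckoo reaction when the player's kill count changes.
; After post-res-in it watches the AthenaKills stat and stores each new value as KillCount.
; Above zero, it spawns a debounced emitter at backpack_root with attached cuckoo sound, and calls the pop-out on the backpack's AnimBP.
; PlayFXandSound duplicates the effect and sound, with no caller in this graph.
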(event EventOnPostResIn
  (|CallParentFunction)
  (bind (_statname _statvalue) (Cosmetics|WatchCosmeticStat (CustomCharacter|GetAssociatedPlayerPawn) "(TagName=\"StatManager.AthenaKills\")" 0 false)
    (:OnStatChanged
      (bind _output_get (Variables|Default|SetKillCount (Math|Conversions|ToFloat(Integer) _statvalue)))
      (if (> _output_get 0.0)
        (bind _returnvalue (PartModifier|SpawnSystem|SpawnEmitterAttachedPartModifier(Legacy) :EmitterTemplate "/Game/Accessories/FORT_Backpacks/M_MED_Octoberfest_Backpack/FX/P_BackPack_Octoberfest.P_BackPack_Octoberfest" :AttachToComponent (CustomCharacter|GetSkeletalMeshForPartOnPlayerPawn) :AttachPointName "backpack_root" :Scale "1.000000,1.000000,1.000000" :LocationType "SnapToTarget" :bAutoDestroy true :PoolingMethod "None" :bAutoActivate true :bMatchPawnCustomDepth true :bDebounce true :DebounceTime -1.0)
          (:Success
            (Audio|SpawnSoundAttached "/Game/Athena/Sounds/BackBlings/BackBling_Octoberfest_Cuckoo_Cue.BackBling_Octoberfest_Cuckoo_Cue" _returnvalue))
          (:Failure
            _))
        (bind _asm_med_octoberfest_backpack_skeleton_anim_blueprint (Utilities|Casting|CastToM_MED_Octoberfest_Backpack_Skeleton_AnimBlueprint (Components|SkeletalMesh|GetAnimInstance (Customization|GetSkeletalMeshforPartType (CustomCharacter|GetAssociatedPlayerPawn) (CustomCharacter|GetCharacterPartType) :!class /Script/FortniteGame.FortPlayerPawn))))
        (Cuckoo|PopOutCuckooonKill _asm_med_octoberfest_backpack_skeleton_anim_blueprint)))))

(event Custom|PlayFXandSound
  (bind _returnvalue (PartModifier|SpawnSystem|SpawnEmitterAttachedPartModifier(Legacy) :EmitterTemplate "/Game/Accessories/FORT_Backpacks/M_MED_Octoberfest_Backpack/FX/P_BackPack_Octoberfest.P_BackPack_Octoberfest" :AttachToComponent (CustomCharacter|GetSkeletalMeshForPartOnPlayerPawn) :AttachPointName "backpack_root" :Scale "1.000000,1.000000,1.000000" :LocationType "SnapToTarget" :bAutoDestroy true :PoolingMethod "None" :bAutoActivate true :bMatchPawnCustomDepth true :bDebounce true :DebounceTime -1.0)
    (:Success
      (Audio|SpawnSoundAttached "/Game/Athena/Sounds/BackBlings/BackBling_Octoberfest_Cuckoo_Cue.BackBling_Octoberfest_Cuckoo_Cue" _returnvalue))))

(fn ConstructionScript)
