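; Cosmetic character part modifier for OrnamentSoldier ShockAndAwe that makes the player's materials react to electrocution.
; Curie state attach and detach events switch on the gameplay tag, handling only the Electrocution elemental status.
; Both set scalar parameters on every mesh component of the associated pawn.
; Attach raises increment speed and emissive blink brightness and enables vertex displacement; detach applies hard-coded calmer values and disables displacement.
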(event EventOnCurieStateAttached
  (params (Tag :type Gameplay%20Tag%20Structure%20%28by%20ref%29))
  (bind _returnvalue (CustomCharacter|GetAssociatedPlayerPawn))
  (bind _returnvalue_2 (Actor|GetComponentsByClass _returnvalue "/Script/Engine.MeshComponent"))
  (switch Utilities|FlowControl|Switch|SwitchonGameplayTag Tag
    (:Curie.State.ElementalStatus.Electrocution
      (Rendering|Material|SetScalarParameterValueonMaterials _returnvalue_2 "IncrementSpeed" 6.0)
      (Rendering|Material|SetScalarParameterValueonMaterials _returnvalue_2 "Emissive_BlinkingBrightness" 2e+02)
      (Rendering|Material|SetScalarParameterValueonMaterials _returnvalue_2 "BL_EnableVertDisplacement" 1.0))))

(event EventOnCurieStateDetached
  (params (Tag :type Gameplay%20Tag%20Structure%20%28by%20ref%29))
  (bind _returnvalue (CustomCharacter|GetAssociatedPlayerPawn))
  (bind _returnvalue_2 (Actor|GetComponentsByClass _returnvalue "/Script/Engine.MeshComponent"))
  (switch Utilities|FlowControl|Switch|SwitchonGameplayTag Tag
    (:Curie.State.ElementalStatus.Electrocution
      (Rendering|Material|SetScalarParameterValueonMaterials _returnvalue_2 "IncrementSpeed" 0.75)
      (Rendering|Material|SetScalarParameterValueonMaterials _returnvalue_2 "Emissive_BlinkingBrightness" 15.0)
      (Rendering|Material|SetScalarParameterValueonMaterials _returnvalue_2 "BL_EnableVertDisplacement" 0.0))))

(fn ConstructionScript
  (|CallParentFunction))
