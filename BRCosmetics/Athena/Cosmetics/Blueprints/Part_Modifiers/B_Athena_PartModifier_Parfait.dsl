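; Cosmetic part modifier for Parfait that makes a character part react with material, particle and audio effects as eliminations are recorded.
; A stat change with a value of at least one triggers all three effects.
; The material receives game time as a reactivity timestamp, while the idle VFX's "PowersActive" flag clears after a retriggerable delay.
; In the front end, post-resource initialization previews material and VFX without sound, followed by a cycle delay.
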(event EventOnPostResIn
  ;[0.007499 0.2122 0.6584] FrontEnd Stage cycling
  (|CallParentFunction)
  (if (CustomCosmeticModifier|IsinFrontEnd)
    (CallFunction|TrifleParfaitReactivity)
    (CallFunction|TriggerFX)
    (Utilities|FlowControl|Delay (Variables|Default|GetFrontendFXCycleDelay))))

(event EventOnStatChanged
  (params (State%20Name :type Name :default "None") (Stat%20Value :type Integer))
  ;[0.272 0.8177 0.1576] Reactivity logic
  (|CallParentFunction State%20Name Stat%20Value)
  (Variables|Default|SetCurrent%20Eliminations Stat%20Value)
  (if (>= (Variables|Default|GetCurrent%20Eliminations) 1)
    (Utilities|FlowControl|Sequence
      (:then_0
        (CallFunction|TriggerFX))
      (:then_1
        (CallFunction|TriggerAudio))
      (:then_2
        (CallFunction|TrifleParfaitReactivity)))))

(event Custom|Trifle%20Parfait%20Reactivity
  ;[1 0.9907 0] Material reactivity
  (Rendering|Material|SetScalarParameterValueonMaterials (CustomCharacter|GetSkeletalMeshForPartOnPlayerPawn) "TP_CurrentTime_Reactivity" (Utilities|Time|GetGameTimeinSeconds)))

(event Custom|Trigger%20FX
  (bind _returnvalue (CallFunction|GetIdleVFXComponent))
  ;[1 0.4948 0.4948] VFX
  (Effects|Components|ParticleSystem|SetBoolParameter _returnvalue "PowersActive" true)
  (Components|Activation|Activate _returnvalue true)
  (Utilities|FlowControl|RetriggerableDelay (Variables|Default|GetDuration))
  (Effects|Components|ParticleSystem|SetBoolParameter _returnvalue "PowersActive" false))

(event Custom|Trigger%20Audio
  ;[0 0 0] Audio (note: we do not want this triggering in the front-end, so it's unhooked from there)
  (Audio|SpawnSoundAttached (Variables|Default|GetSound) (CustomCharacter|GetSkeletalMeshForPartOnPlayerPawn)))

(fn ConstructionScript)
