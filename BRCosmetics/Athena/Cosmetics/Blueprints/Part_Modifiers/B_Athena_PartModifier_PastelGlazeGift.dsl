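; Cosmetic part modifier for the PastelGlazeGift character part that drives a reactive material look and idle VFX.
; In the front end, the FE Preview timeline animates PGG_Reactivity and toggles the idle Niagara effect by Material variant, on for Mat0 and off for Mat1.
; In game, normalized accumulated velocity drives PGG_Reactivity instead.
; Construction enables custom depth on the part mesh.
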(event EventOnPostResIn
  (bind _returnvalue (CallFunction|GetIdleVFXComponent))
  (bind _returnvalue_2 (CustomCharacter|GetSkeletalMeshForPartOnPlayerPawn))
  (bind _returnvalue_3 (CallFunction|GetIdleVFXComponent))
  ; FE preview
  (|CallParentFunction)
  (if (CustomCosmeticModifier|IsinFrontEnd)
    (Rendering|Material|SetScalarParameterValueonMaterials _returnvalue_2 "DiamondEmissiveIntensity" 8.0)
    (bind (_direction _alpha) (Timeline|FE%20Preview PlayFromStart)
      (:Update
        (Rendering|Material|SetScalarParameterValueonMaterials _returnvalue_2 "PGG_Reactivity" _alpha)
        (Effects|Components|ParticleSystem|SetBoolParameter _returnvalue "bIsFrontEnd" true)
        (Effects|Components|ParticleSystem|SetBoolParameter _returnvalue "bIsFrontEndPreview" true)
        ;[0.3024 0.1295 0.6719 0.5] FX - Reactivity check
        (bind _asfort_player_pawn (Utilities|Casting|CastToFortPlayerPawn (Actor|GetOwner)))
        (bind _index (CosmeticLoadout|GetVariantChannelActiveIndexforSlotActor "(TagName=\"LoadoutSlot.Character\")" _asfort_player_pawn "(TagName=\"Cosmetics.Variant.Channel.Material\")" (Utilities|Array|MakeArray "(TagName=\"Cosmetics.Variant.Property.Mat0\")" "(TagName=\"Cosmetics.Variant.Property.Mat1\")") false false))
        (if (Utilities|Select :Index _index :Option%200 true :Option%201 false)
          (bind _asniagara_particle_system_component (Utilities|Casting|CastToNiagaraComponent _returnvalue_3))
          (Components|Activation|Activate _asniagara_particle_system_component false)
          (else
            (bind _asniagara_particle_system_component_2 (Utilities|Casting|CastToNiagaraComponent _returnvalue_3))
            (Components|Activation|Deactivate _asniagara_particle_system_component_2)))))
    (else
      (Effects|Components|ParticleSystem|SetBoolParameter _returnvalue "bIsFrontEnd" false)
      (Effects|Components|ParticleSystem|SetBoolParameter _returnvalue "bIsFrontEndPreview" false))))

(event EventTick
  (params (DeltaSeconds :type Float%20%28single-precision%29))
  ; In game reactivity
  (|CallParentFunction DeltaSeconds))

(event EventOnAccumulatedVelocity
  (params (NormalizedAccumulatedVelocity :type Float%20%28double-precision%29) (AccumulatedVelocity :type Float%20%28double-precision%29))
  ; In game reactivity
  (Rendering|Material|SetScalarParameterValueonMaterials (CustomCharacter|GetSkeletalMeshForPartOnPlayerPawn) "PGG_Reactivity" NormalizedAccumulatedVelocity))

(event EventOnNewWeapons
  (params (WeaponsAsItems :type Array%20of%20Fort%20Item%20Object%20References)))

(event EventOnCurieStateAttached
  (params (Tag :type Gameplay%20Tag%20Structure%20%28by%20ref%29)))

(event EventOnCurieStateDetached
  (params (Tag :type Gameplay%20Tag%20Structure%20%28by%20ref%29)))

(event EventOnBeginSkydiving
  (|CallParentFunction) ; This node is disabled and will not be called.\nDrag off pins to build functionality.
)

(event EventOnEndSkydiving
  (|CallParentFunction) ; This node is disabled and will not be called.\nDrag off pins to build functionality.
)

(event EventOnEnterVehicle
  (params (VehicleEntered :type Fort%20Vehicle%20Interface%20Interface%20%28by%20ref%29))
  (|CallParentFunction VehicleEntered) ; This node is disabled and will not be called.\nDrag off pins to build functionality.
)

(event EventOnExitVehicle
  (|CallParentFunction) ; This node is disabled and will not be called.\nDrag off pins to build functionality.
)

(event EventOnTargetingOrFirstPersonCameraChanged)

(event EventOnHitPawn
  (params (SelfActor :type Actor%20Object%20Reference) (OtherActor :type Actor%20Object%20Reference) (NormalImpulse :type Vector :default "0, 0, 0") (Hit :type Hit%20Result%20Structure%20%28by%20ref%29)))

(fn ConstructionScript
  (|CallParentFunction)
  (Rendering|SetRenderCustomDepth (CustomCharacter|GetSkeletalMeshForPartOnPlayerPawn) true))
