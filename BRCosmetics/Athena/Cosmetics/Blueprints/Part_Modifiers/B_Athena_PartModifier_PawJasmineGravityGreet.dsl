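; Part modifier for the PawJasmineGravity Greet skin that plays a glitch reaction on one material variant.
; After res-in it reads hood and material variants; positive stat changes in-game, or a looping front-end timer, trigger the effect.
; The effect spawns attached Niagara and sets face and body overlay materials, with timelines stepping glitch blocks in, then out after a delay.
; Particles receive hidden hand/foot flags, plus a low-LOD flag on tick in-game.
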(event Custom|FE%20Glitch%20Reactivity
  (bind _returnvalue (CustomCharacter|GetAssociatedPlayerPawn))
  (bind _returnvalue_2 (Customization|GetSkeletalMeshforPartType _returnvalue "Face" :!class /Script/FortniteGame.FortPlayerPawn))
  (bind _returnvalue_3 (Customization|GetSkeletalMeshforPartType _returnvalue "Body" :!class /Script/FortniteGame.FortPlayerPawn))
  (Rendering|Material|SetOverlayMaterial _returnvalue_3 "/BRCosmetics/Characters/Player/Male/Medium/Bodies/M_MED_PawJasmineGravity/Skins/Greet/Materials/MI_PawJasmineGravity_Greet_Body_Overlay.MI_PawJasmineGravity_Greet_Body_Overlay" false 0 :!class /Script/Engine.MeshComponent)
  (Rendering|Material|SetOverlayMaterial _returnvalue_2 "/BRCosmetics/Characters/Player/Male/Medium/Bodies/M_MED_PawJasmineGravity/Skins/Greet/Materials/MI_PawJasmineGravity_Greet_FaceAcc_Overlay.MI_PawJasmineGravity_Greet_FaceAcc_Overlay" false 0 :!class /Script/Engine.MeshComponent)
  (Class|FortCosmeticMaterialBlueprintLibrary|SetScalarParameterValueonOverlayMaterial _returnvalue_3 "Glitch Opacity" 1.0)
  (Class|FortCosmeticMaterialBlueprintLibrary|SetScalarParameterValueonOverlayMaterial _returnvalue_2 "Glitch Opacity" 1.0)
  (Class|FortCosmeticMaterialBlueprintLibrary|SetScalarParameterValueonOverlayMaterial _returnvalue_2 "Hood On" (Variables|Default|GetHood%20On))
  (Class|FortCosmeticMaterialBlueprintLibrary|SetScalarParameterValueonOverlayMaterial _returnvalue_2 "Hood Off" (Variables|Default|GetHood%20Off))
  (Rendering|Material|SetScalarParameterValueonMaterials _returnvalue_3 "Enable Reactivity?" 1.0)
  (bind (_direction _block_fade) (Timeline|Glitch%20Blocks PlayFromStart)
    (:Update
      (Class|FortCosmeticMaterialBlueprintLibrary|SetScalarParameterValueonOverlayMaterial _returnvalue_2 "Glitch Block - Steps" _block_fade)
      (Class|FortCosmeticMaterialBlueprintLibrary|SetScalarParameterValueonOverlayMaterial _returnvalue_3 "Glitch Block - Steps" _block_fade))
    (:Finished
      (Utilities|FlowControl|Delay 3.0)
      (bind (_direction_2 _block_fade_2) (Timeline|Glitch%20Blocks%20-%20Reversed ReverseFromEnd)
        (:Update
          (Class|FortCosmeticMaterialBlueprintLibrary|SetScalarParameterValueonOverlayMaterial _returnvalue_2 "Glitch Block - Steps" _block_fade_2)
          (Class|FortCosmeticMaterialBlueprintLibrary|SetScalarParameterValueonOverlayMaterial _returnvalue_3 "Glitch Block - Steps" _block_fade_2))
        (:Finished
          (Class|FortCosmeticMaterialBlueprintLibrary|SetScalarParameterValueonOverlayMaterial _returnvalue_2 "Glitch Opacity")
          (Class|FortCosmeticMaterialBlueprintLibrary|SetScalarParameterValueonOverlayMaterial _returnvalue_3 "Glitch Opacity")
          (Rendering|Material|SetScalarParameterValueonMaterials _returnvalue_3 "Enable Reactivity?" 0.0))))))

(event EventOnStatChanged
  (params (State%20Name :type Name :default "None") (Stat%20Value :type Integer))
  (bind _returnvalue (CustomCharacter|GetAssociatedPlayerPawn))
  (bind _returnvalue_2 (Customization|GetSkeletalMeshforPartType _returnvalue "Face" :!class /Script/FortniteGame.FortPlayerPawn))
  (bind _returnvalue_3 (Customization|GetSkeletalMeshforPartType _returnvalue "Body" :!class /Script/FortniteGame.FortPlayerPawn))
  (|CallParentFunction State%20Name Stat%20Value)
  (if (CustomCosmeticModifier|IsinFrontEnd)
    _
    (elif (> Stat%20Value 0)
      (if (Variables|Default|GetIs%20Reactive)
        (bind _returnvalue_4 (PartModifier|SpawnSystem|SpawnSystemAttachedPartModifier "/BRCosmetics/Characters/Player/Male/Medium/Bodies/M_MED_PawJasmineGravity/FX/NS_Character_PawJasmineGravityGreet_Reactivity.NS_Character_PawJasmineGravityGreet_Reactivity" (CustomCharacter|GetSkeletalMeshForPartOnPlayerPawn) "root")
          (:Success
            (Effects|Components|ParticleSystem|SetBoolParameter _returnvalue_4 "bisFootHidden" (Variables|HiddenPartCheck|GetbIsFootHidden))
            (Effects|Components|ParticleSystem|SetBoolParameter _returnvalue_4 "bisHandHidden" (Variables|HiddenPartCheck|GetbIsHandHidden))
            (if (Variables|Default|GetIs%20Reactive)
              (Rendering|Material|SetOverlayMaterial _returnvalue_3 "/BRCosmetics/Characters/Player/Male/Medium/Bodies/M_MED_PawJasmineGravity/Skins/Greet/Materials/MI_PawJasmineGravity_Greet_Body_Overlay.MI_PawJasmineGravity_Greet_Body_Overlay" false 0 :!class /Script/Engine.MeshComponent)
              (Rendering|Material|SetOverlayMaterial _returnvalue_2 "/BRCosmetics/Characters/Player/Male/Medium/Bodies/M_MED_PawJasmineGravity/Skins/Greet/Materials/MI_PawJasmineGravity_Greet_FaceAcc_Overlay.MI_PawJasmineGravity_Greet_FaceAcc_Overlay" false 0 :!class /Script/Engine.MeshComponent)
              (Class|FortCosmeticMaterialBlueprintLibrary|SetScalarParameterValueonOverlayMaterial _returnvalue_3 "Glitch Opacity" 1.0)
              (Class|FortCosmeticMaterialBlueprintLibrary|SetScalarParameterValueonOverlayMaterial _returnvalue_2 "Glitch Opacity" 1.0)
              (Class|FortCosmeticMaterialBlueprintLibrary|SetScalarParameterValueonOverlayMaterial _returnvalue_2 "Hood On" (Variables|Default|GetHood%20On))
              (Class|FortCosmeticMaterialBlueprintLibrary|SetScalarParameterValueonOverlayMaterial _returnvalue_2 "Hood Off" (Variables|Default|GetHood%20Off))
              (Rendering|Material|SetScalarParameterValueonMaterials _returnvalue_3 "Enable Reactivity?" 1.0)
              (Timeline|Glitch%20Blocks PlayFromStart))))))))

(event EventOnPostResIn
  (|CallParentFunction)
  (bind _index (CosmeticLoadout|GetVariantChannelActiveIndexforSlotActor "(TagName=\"LoadoutSlot.Character\")" (CustomCharacter|GetAssociatedPlayerPawn) "(TagName=\"Cosmetics.Variant.Channel.Parts\")" (Utilities|Array|MakeArray "(TagName=\"Cosmetics.Variant.Property.Stage1\")" "(TagName=\"Cosmetics.Variant.Property.Stage2\")")))
  (Variables|Default|SetHood%20On (Math|Float|SelectFloat 1.0 0.0 (Math|Conversions|ToBoolean(Integer) _index)))
  (Variables|Default|SetHood%20Off (Math|Float|SelectFloat 0.0 1.0 (Math|Conversions|ToBoolean(Integer) _index)))
  (bind _index_2 (CosmeticLoadout|GetVariantChannelActiveIndexforSlotActor "(TagName=\"LoadoutSlot.Character\")" (CustomCharacter|GetAssociatedPlayerPawn) "(TagName=\"Cosmetics.Variant.Channel.Material\")" (Utilities|Array|MakeArray "(TagName=\"Cosmetics.Variant.Property.Mat1\")" "(TagName=\"Cosmetics.Variant.Property.Mat2\")" "(TagName=\"Cosmetics.Variant.Property.Mat3\")")))
  (bind _output_get (Variables|Default|SetIs%20Reactive (== _index_2 1)))
  (if (and _output_get (CustomCosmeticModifier|IsinFrontEnd))
    (if (Utilities|IsValid (CustomCharacter|GetSkeletalMeshForPartOnPlayerPawn) :!class /Script/Engine.KismetSystemLibrary)
      (Utilities|Time|SetTimerbyEvent (delegate Custom|FE%20Reactivity) 6.0 true true))))

(event Custom|FE%20Reactivity
  (PartModifier|SpawnSystem|SpawnSystemAttachedPartModifier "/BRCosmetics/Characters/Player/Male/Medium/Bodies/M_MED_PawJasmineGravity/FX/NS_Character_PawJasmineGravityGreet_Reactivity.NS_Character_PawJasmineGravityGreet_Reactivity" (CustomCharacter|GetSkeletalMeshForPartOnPlayerPawn) "root"
    (:Success
      _)
    (:Failure
      _))
  (CallFunction|FEGlitchReactivity))

(event EventTick
  (params (DeltaSeconds :type Float%20%28single-precision%29))
  (|CallParentFunction 0.0)
  (if (and (Utilities|IsValid (stale EventOnStatChanged _returnvalue_4) :!class /Script/Engine.KismetSystemLibrary) (not (CustomCosmeticModifier|IsinFrontEnd)))
    (bind _asskeletal_mesh_component (Utilities|Casting|CastToSkeletalMeshComponent (CustomCharacter|GetSkeletalMeshForPartOnPlayerPawn)))
    (bind _ascustom_character_part_anim_instance (Utilities|Casting|CastToCustomCharacterPartAnimInstance (Components|SkeletalMesh|GetAnimInstance _asskeletal_mesh_component)))
    (Effects|Components|ParticleSystem|SetBoolParameter (stale EventOnStatChanged _returnvalue_4) "bisLowLOD" (>= (Variables|/Script/FortniteGame.CustomCharacterPartAnimInstance|GetCurrentLOD _ascustom_character_part_anim_instance) 3))))

(fn ConstructionScript
  (|CallParentFunction))
