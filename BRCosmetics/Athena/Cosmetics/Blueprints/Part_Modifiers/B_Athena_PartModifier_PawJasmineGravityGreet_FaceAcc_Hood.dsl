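; Part modifier for the PawJasmineGravityGreet hood face accessory that spawns reactive hood-up or hood-down Niagara effects on the character.
; On PostResIn, the Parts variant sets hood state and the second Material variant enables reactivity.
; In game, positive stat changes spawn the matching effect at the mesh root and pass it the head-hidden flag.
; In the front end, a looping timer replays the effect without that flag.
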(event EventOnStatChanged
  (params (State%20Name :type Name :default "None") (Stat%20Value :type Integer))
  (bind _returnvalue (CustomCharacter|GetSkeletalMeshForPartOnPlayerPawn))
  (|CallParentFunction State%20Name Stat%20Value)
  (if (CustomCosmeticModifier|IsinFrontEnd)
    _
    (elif (> Stat%20Value 0)
      (if (Variables|Default|Getis%20Reactive)
        (if (Variables|Default|Getis%20Hood%20Down)
          (bind _returnvalue_2 (PartModifier|SpawnSystem|SpawnSystemAttachedPartModifier "/BRCosmetics/Characters/Player/Male/Medium/Bodies/M_MED_PawJasmineGravity/FX/NS_Character_PawJasmineGravityGreet_Reactivity_HoodDown.NS_Character_PawJasmineGravityGreet_Reactivity_HoodDown" _returnvalue "root")
            (:Success
              (Effects|Components|ParticleSystem|SetBoolParameter _returnvalue_2 "bisHeadHidden" (Variables|HiddenPartCheck|GetbIsHeadHidden))))
          (else
            (bind _returnvalue_3 (PartModifier|SpawnSystem|SpawnSystemAttachedPartModifier "/BRCosmetics/Characters/Player/Male/Medium/Bodies/M_MED_PawJasmineGravity/FX/NS_Character_PawJasmineGravityGreet_Reactivity_HoodUp.NS_Character_PawJasmineGravityGreet_Reactivity_HoodUp" _returnvalue "root")
              (:Success
                (Effects|Components|ParticleSystem|SetBoolParameter _returnvalue_3 "bisHeadHidden" (Variables|HiddenPartCheck|GetbIsHeadHidden))))))))))

(event EventOnPostResIn
  (|CallParentFunction)
  (bind _index (CosmeticLoadout|GetVariantChannelActiveIndexforSlotActor "(TagName=\"LoadoutSlot.Character\")" (CustomCharacter|GetAssociatedPlayerPawn) "(TagName=\"Cosmetics.Variant.Channel.Parts\")" (Utilities|Array|MakeArray "(TagName=\"Cosmetics.Variant.Property.Stage1\")" "(TagName=\"Cosmetics.Variant.Property.Stage2\")")))
  (Variables|Default|Setis%20Hood%20Down (== _index 0))
  (bind _index_2 (CosmeticLoadout|GetVariantChannelActiveIndexforSlotActor "(TagName=\"LoadoutSlot.Character\")" (CustomCharacter|GetAssociatedPlayerPawn) "(TagName=\"Cosmetics.Variant.Channel.Material\")" (Utilities|Array|MakeArray "(TagName=\"Cosmetics.Variant.Property.Mat1\")" "(TagName=\"Cosmetics.Variant.Property.Mat2\")" "(TagName=\"Cosmetics.Variant.Property.Mat3\")")))
  (bind _output_get (Variables|Default|Setis%20Reactive (== _index_2 1)))
  (if (and (CustomCosmeticModifier|IsinFrontEnd) _output_get)
    (if (Utilities|IsValid (CustomCharacter|GetSkeletalMeshForPartOnPlayerPawn) :!class /Script/Engine.KismetSystemLibrary)
      (Utilities|Time|SetTimerbyEvent (delegate Custom|FE%20Reactivity) 6.0 true true))))

(event Custom|FE%20Reactivity
  (bind _returnvalue (CustomCharacter|GetSkeletalMeshForPartOnPlayerPawn))
  (if (Variables|Default|Getis%20Hood%20Down)
    (PartModifier|SpawnSystem|SpawnSystemAttachedPartModifier "/BRCosmetics/Characters/Player/Male/Medium/Bodies/M_MED_PawJasmineGravity/FX/NS_Character_PawJasmineGravityGreet_Reactivity_HoodDown.NS_Character_PawJasmineGravityGreet_Reactivity_HoodDown" _returnvalue "root")
    (else
      (PartModifier|SpawnSystem|SpawnSystemAttachedPartModifier "/BRCosmetics/Characters/Player/Male/Medium/Bodies/M_MED_PawJasmineGravity/FX/NS_Character_PawJasmineGravityGreet_Reactivity_HoodUp.NS_Character_PawJasmineGravityGreet_Reactivity_HoodUp" _returnvalue "root"))))

(event EventOnDayPhaseChanged
  (params (CurrentDayPhase :type EFortDayPhase%20Enum :default "Morning") (PreviousDayPhase :type EFortDayPhase%20Enum :default "Morning") (bAtCreation :type Boolean)))

(event EventOnNewPlayerWeaponType
  (params (NewType :type EFortItemType%20Enum :default "WorldItem") (OldType :type EFortItemType%20Enum :default "WorldItem"))
  (|CallParentFunction NewType OldType) ; This node is disabled and will not be called.\nDrag off pins to build functionality.
)

(event EventOnWeaponChanged
  (params (NewWeapon :type Fort%20Weapon%20Object%20Reference) (OldWeapon :type Fort%20Weapon%20Object%20Reference))
  (|CallParentFunction NewWeapon OldWeapon) ; This node is disabled and will not be called.\nDrag off pins to build functionality.
)

(event EventOnWeaponFired)

(event EventOnNewWeapons
  (params (WeaponsAsItems :type Array%20of%20Fort%20Item%20Object%20References)))

(event EventOnCurieStateAttached
  (params (Tag :type Gameplay%20Tag%20Structure%20%28by%20ref%29)))

(event EventOnCurieStateDetached
  (params (Tag :type Gameplay%20Tag%20Structure%20%28by%20ref%29)))

(event EventOnBeginSkydiving
  (|CallParentFunction) ; This node is disabled and will not be called.\nDrag off pins to build functionality.
)

(event EventOnEndSkydiving
  (|CallParentFunction) ; This node is disabled and will not be called.\nDrag off pins to build functionality.
)

(event EventOnEnterVehicle
  (params (VehicleEntered :type Fort%20Vehicle%20Interface%20Interface%20%28by%20ref%29))
  (|CallParentFunction VehicleEntered) ; This node is disabled and will not be called.\nDrag off pins to build functionality.
)

(event EventOnExitVehicle
  (|CallParentFunction) ; This node is disabled and will not be called.\nDrag off pins to build functionality.
)

(event EventOnTargetingOrFirstPersonCameraChanged)

(event EventOnHitPawn
  (params (SelfActor :type Actor%20Object%20Reference) (OtherActor :type Actor%20Object%20Reference) (NormalImpulse :type Vector :default "0, 0, 0") (Hit :type Hit%20Result%20Structure%20%28by%20ref%29)))

(fn ConstructionScript
  (|CallParentFunction))
