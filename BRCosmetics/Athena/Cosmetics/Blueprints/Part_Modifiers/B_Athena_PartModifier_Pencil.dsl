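; Character part modifier for the Pencil cosmetic that sets outline and shading style from the cosmetic items' meta tags.
; On post-resolve init (EventOnPostResIn), it gathers tags into OutlineOff, PBR and Toon flags, then updates the part's skeletal mesh.
; It swaps the overlay material for outlines and sets PBR_Shading, with PBR taking priority over Toon.
; Toon in the front end also gets dynamic materials with an inverse LUT, marked as temporary.
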(event EventOnPostResIn
  (bind _returnvalue (CustomCharacter|GetSkeletalMeshForPartOnPlayerPawn))
  (bind _returnvalue_2 (CustomCharacter|GetSkeletalMeshForPartOnPlayerPawn))
  (bind _returnvalue_3 (CustomCharacter|GetSkeletalMeshForPartOnPlayerPawn))
  (bind _returnvalue_4 (CustomCharacter|GetAssociatedPlayerPawn))
  (|CallParentFunction)
  ;[0 0.3147 1] Check Meta Tags of CID
  (for _array_element (Variables|Default|GetCosmetic%20Item)
    (if (GameplayTags|HasTag (Cosmetics|GatherMetaTags _returnvalue_4 _array_element :!class /Script/FortniteGame.FortPlayerPawn) "(TagName=\"Cosmetics.MetaTags.Pencil.OutlinesOff\")" true)
      (Variables|Default|SetOutlineOff true))
    (if (GameplayTags|HasTag (Cosmetics|GatherMetaTags _returnvalue_4 _array_element :!class /Script/FortniteGame.FortPlayerPawn) "(TagName=\"Cosmetics.MetaTags.Pencil.PBR\")" true)
      (Variables|Default|SetPBR true))
    (if (GameplayTags|HasTag (Cosmetics|GatherMetaTags _returnvalue_4 _array_element :!class /Script/FortniteGame.FortPlayerPawn) "(TagName=\"Cosmetics.MetaTags.Pencil.Toon\")")
      (Variables|Default|SetToon true)))
  (Utilities|FlowControl|Sequence
    (:then_0
      ;[1 1 1] Front End LUT, Remove after creating auto LUT tool
      (if (and (CustomCosmeticModifier|IsinFrontEnd) (Variables|Default|GetToon))
        (for (_array_element_2 _array_index) (Rendering|Material|GetMaterials _returnvalue_3)
          (bind _returnvalue_5 (Rendering|Material|CreateDynamicMaterialInstance _returnvalue_3 _array_index _array_element_2 "None" :!class /Script/Engine.PrimitiveComponent))
          (Rendering|Material|SetTextureParameterValue _returnvalue_5 "LUT Volume Texture" "/Game/Packages/Fortress_SharedMaterials/Anime_Master_Material/VT_InverseLUT_FE.VT_InverseLUT_FE")
          (Rendering|Material|SetScalarParameterValue _returnvalue_5 "LUT Multiplier" 1.0 :!class /Script/Engine.MaterialInstanceDynamic)
          (Rendering|Material|SetScalarParameterValue _returnvalue_5 "Mobile LUT Brightness" 1.0 :!class /Script/Engine.MaterialInstanceDynamic)
          (Rendering|Material|SetScalarParameterValue _returnvalue_5 "use 33 Resolution" 0.0 :!class /Script/Engine.MaterialInstanceDynamic))))
    (:then_1
      ;[1 1 1] Outlines
      (Utilities|IsValid (Variables|Default|GetOverlayMaterial)
        (:Is%20Valid
          _)
        (:Is%20Not%20Valid
          (Variables|Default|SetOverlayMaterial (Rendering|Material|GetOverlayMaterial _returnvalue_2 false 0 :!class /Script/Engine.MeshComponent))))
      (if (Variables|Default|GetOutlineOff)
        (Rendering|Material|SetOverlayMaterial _returnvalue_2 "/Game/Global/Materials/M_Overlay_VertexCrunch.M_Overlay_VertexCrunch" false 0 :!class /Script/Engine.MeshComponent) ; Disable Overlay Materials
        (else
          (Rendering|Material|SetOverlayMaterial _returnvalue_2 (Variables|Default|GetOverlayMaterial) false 0 :!class /Script/Engine.MeshComponent) ; Activate Overlay Materials
        )))
    (:then_2
      ;[1 1 1] Shading
      (if (Variables|Default|GetPBR)
        (Rendering|Material|SetScalarParameterValueonMaterials _returnvalue "PBR_Shading" 1.0) ; PBR Shading On
        (elif (Variables|Default|GetToon)
          (Rendering|Material|SetScalarParameterValueonMaterials _returnvalue "PBR_Shading" 0.0) ; Toon Shading On
        )))))

(fn ConstructionScript
  (|CallParentFunction))
