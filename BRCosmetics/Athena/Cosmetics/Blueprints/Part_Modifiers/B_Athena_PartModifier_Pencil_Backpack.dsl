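; Cosmetic part modifier for the Pencil backpack that flashes reactive icons on its material and drives a Chibi blink animation.
; In game it watches health gain, kill and chest-opening stats, setting an icon location per event and playing reactivity start and end timelines.
; Emote start shows its own icon and holds reactivity until the emote stops.
; In the front end a preview timeline animates icon and reactivity instead.
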(event EventOnCharacterCustomizationFinished
  (params (CharacterPartInfo :type Array%20of%20Applied%20Character%20Part%20Info%20Structures))
  ;[1 1 1] Chibi blinking anim
  (|CallParentFunction CharacterPartInfo)
  (Utilities|FlowControl|Sequence
    (:then_0
      (bind (_direction _blink) (Timeline|Blink PlayFromStart)
        (:Update
          (Rendering|Material|SetScalarParameterValueonMaterials (CustomCharacter|GetSkeletalMeshForPartOnPlayerPawn) "Chibi_Blink" _blink))))))

(event EventOnPostResIn
  (bind _returnvalue (CustomCharacter|GetSkeletalMeshForPartOnPlayerPawn))
  (bind _returnvalue_2 (CustomCharacter|GetSkeletalMeshForPartOnPlayerPawn))
  (bind _returnvalue_3 (CustomCharacter|GetSkeletalMeshForPartOnPlayerPawn))
  (bind _returnvalue_4 (CustomCharacter|GetAssociatedPlayerPawn))
  (|CallParentFunction)
  (if (CustomCosmeticModifier|IsinFrontEnd)
    ;[0 0.4625 1] FE
    (bind (_direction _location _reactivity) (Timeline|FE%20Reactivity%20Preview PlayFromStart)
      (:Update
        (Rendering|Material|SetScalarParameterValueonMaterials _returnvalue_2 "PencilBackpack_Reactivity" _reactivity)
        (Rendering|Material|SetVectorParameterValueonMaterials _returnvalue_2 "PencilBackpack_IconLocation" (Utilities|Array|Get(acopy) (Variables|Default|GetLocations) (Math|Float|Truncate _location)))))
    (else
      (Utilities|FlowControl|Sequence
        (:then_0
          (Cosmetics|WatchCosmeticStat _returnvalue_4 "(TagName=\"StatManager.GainedHealthTimes\")" 0 false
            (:OnStatChanged
              (Rendering|Material|SetVectorParameterValueonMaterials _returnvalue_3 "PencilBackpack_IconLocation" (Utilities|Array|Get(acopy) (Variables|Default|GetLocations)))
              (bind (_direction_2 _reactivity_2) (Timeline|InGame%20Reactivity%20Start PlayFromStart :NewTime 1.0)
                (:Update
                  (Rendering|Material|SetScalarParameterValueonMaterials _returnvalue "PencilBackpack_Reactivity" _reactivity_2))
                (:Finished
                  (if (not (Variables|Default|GetIsEmoting))
                    (bind (_direction_3 _reactivity_3) (Timeline|InGame%20Reactivity%20End PlayFromStart :NewTime 1.0)
                      (:Update
                        (Rendering|Material|SetScalarParameterValueonMaterials _returnvalue "PencilBackpack_Reactivity" _reactivity_3)))))))))
        (:then_1
          (Cosmetics|WatchCosmeticStat _returnvalue_4 "(TagName=\"StatManager.AthenaKills\")" 0 false
            (:OnStatChanged
              _)))
        (:then_2
          (Cosmetics|WatchCosmeticStat _returnvalue_4 "(TagName=\"StatManager.AIKills\")" 0 false
            (:OnStatChanged
              _)))
        (:then_3
          (Cosmetics|WatchCosmeticStat _returnvalue_4 "(TagName=\"StatManager.OpenChests\")" 0 false
            (:OnStatChanged
              (Rendering|Material|SetVectorParameterValueonMaterials _returnvalue_3 "PencilBackpack_IconLocation" (Utilities|Array|Get(acopy) (Variables|Default|GetLocations) 3))
              (Timeline|InGame%20Reactivity%20Start PlayFromStart)))))
      (Rendering|Material|SetVectorParameterValueonMaterials _returnvalue_3 "PencilBackpack_IconLocation" (Utilities|Array|Get(acopy) (Variables|Default|GetLocations) 1))
      (Timeline|InGame%20Reactivity%20Start PlayFromStart))))

(event EventOnPawnAndPartChanged
  (|CallParentFunction)
  (bind _asfort_player_pawn (Utilities|Casting|CastToFortPlayerPawn (CustomCharacter|GetAssociatedPlayerPawn)))
  (EventDispatchers|BindEventtoOnPawnStartedEmote _asfort_player_pawn (delegate EventOnBeginEmote))
  (EventDispatchers|BindEventtoOnPawnStoppedEmote _asfort_player_pawn (delegate EventOnEndEmote)))

(event EventOnBeginEmote
  (params (MontageItemDef :type Fort%20Item%20Definition%20Object%20Reference) (PawnEmoting :type Fort%20Pawn%20Object%20Reference))
  (bind _returnvalue (CustomCharacter|GetSkeletalMeshForPartOnPlayerPawn))
  (Rendering|Material|SetVectorParameterValueonMaterials (CustomCharacter|GetSkeletalMeshForPartOnPlayerPawn) "PencilBackpack_IconLocation" (Utilities|Array|Get(acopy) (Variables|Default|GetLocations) 2))
  (Variables|Default|SetIsEmoting true)
  (Timeline|InGame%20Reactivity%20Start PlayFromStart))

(event EventOnEndEmote
  (params (MontageItemDef :type Fort%20Item%20Definition%20Object%20Reference) (PawnEmoting :type Fort%20Pawn%20Object%20Reference))
  (Variables|Default|SetIsEmoting false)
  (Timeline|InGame%20Reactivity%20End PlayFromStart))

(fn ConstructionScript
  (|CallParentFunction))
