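; Cosmetic part modifier for PhoenixKick that adjusts the character part's rendering when shown in a frontend context.
; After res-in it enables actor tick only if the cosmetic is displayed in the frontend, and disables it otherwise.
; Each tick calls the parent, sets IsFrontend on the part mesh's dynamic overlay material, and turns on custom depth for the part's mesh on the player pawn.
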(event EventOnPostResIn
  (if (Reactive|Frontend|IsCosmeticObjectDisplayedinFrontendContext 0)
    (Actor|Tick|SetActorTickEnabled true)
    (else
      (Actor|Tick|SetActorTickEnabled false))))

(event EventTick
  (params (DeltaSeconds :type Float%20%28single-precision%29))
  (|CallParentFunction DeltaSeconds)
  (bind _asmaterial_instance_dynamic (Utilities|Casting|CastToMaterialInstanceDynamic (Rendering|Material|GetOverlayMaterial (CustomCharacter|GetSkeletalMeshforPartonOwner) true 0 :!class /Script/Engine.MeshComponent)))
  (Rendering|Material|SetScalarParameterValue _asmaterial_instance_dynamic "IsFrontend" 1.0 :!class /Script/Engine.MaterialInstanceDynamic)
  (Rendering|SetRenderCustomDepth (CustomCharacter|GetSkeletalMeshForPartOnPlayerPawn) true))

(fn ConstructionScript
  (|CallParentFunction))
