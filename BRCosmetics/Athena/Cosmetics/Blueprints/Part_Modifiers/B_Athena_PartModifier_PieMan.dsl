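; Part modifier Blueprint for the PieMan cosmetic that makes its idle VFX react to the pawn jumping and landing.
; Jump and land events call the parent, then set a bIsJumping bool parameter on the idle VFX particle component.
; Landing also pulses SpawnLandedFX on, then off after a short delay, to trigger a one-shot landing effect.
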(event EventOnPawnJumped
  (|CallParentFunction)
  (Effects|Components|ParticleSystem|SetBoolParameter (CallFunction|GetIdleVFXComponent) "bIsJumping" true))

(event EventOnPawnLanded
  (bind _returnvalue (CallFunction|GetIdleVFXComponent))
  (|CallParentFunction)
  (Effects|Components|ParticleSystem|SetBoolParameter _returnvalue "bIsJumping" false)
  (Effects|Components|ParticleSystem|SetBoolParameter _returnvalue "SpawnLandedFX" true)
  (Utilities|FlowControl|Delay 0.1)
  (Effects|Components|ParticleSystem|SetBoolParameter _returnvalue "SpawnLandedFX" false))

(fn ConstructionScript
  (|CallParentFunction))
