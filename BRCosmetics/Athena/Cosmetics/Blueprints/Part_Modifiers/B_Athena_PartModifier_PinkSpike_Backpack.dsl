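; Cosmetic part modifier for the PinkSpike backpack that tunes its mesh's overlay material.
; After res-in, it runs parent logic and caps the overlay draw distance.
; In frontend display contexts, it also targets a longer distance and disables UseCenteredLines to fix line thickness, creating a dynamic overlay material when none exists.
; The other gameplay event overrides, such as weapon, skydiving and vehicle, are empty.
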(event Custom|CustomEvent
  (|CallParentFunction))

(event EventOnPostResIn
  (bind _returnvalue (Rendering|Material|GetOverlayMaterial (Customization|GetSkeletalMeshforPartType (CustomCharacter|GetAssociatedPlayerPawn) "Backpack" :!class /Script/FortniteGame.FortPlayerPawn) false 0 :!class /Script/Engine.MeshComponent))
  (bind _returnvalue_2 (CustomCharacter|GetAssociatedPlayerPawn))
  (bind _returnvalue_3 (Customization|GetSkeletalMeshforPartType _returnvalue_2 "Backpack" :!class /Script/FortniteGame.FortPlayerPawn))
  (|CallParentFunction)
  (Utilities|FlowControl|Sequence
    (:then_0
      ;[1 1 1] FrontEnd
      (Rendering|Material|SetOverlayMaterialMaxDrawDistance _returnvalue_3 2.5e+03 :!class /Script/Engine.MeshComponent))
    (:then_1
      (if (Reactive|Frontend|IsCosmeticObjectDisplayedinFrontendContext 0)
        (bind _asmaterial_instance_dynamic (Utilities|Casting|CastToMaterialInstanceDynamic _returnvalue)
          (:then
            (Rendering|Material|SetScalarParameterValue _asmaterial_instance_dynamic "UseCenteredLines" 0.0 :!class /Script/Engine.MaterialInstanceDynamic) ; Fixes Thickness Issue in Front End
          )
          (:CastFailed
            (bind _returnvalue_4 (Rendering|Material|CreateDynamicMaterialInstance _returnvalue "None" "None" :!class /Script/Engine.KismetMaterialLibrary))
            (Rendering|Material|SetOverlayMaterial _returnvalue_3 _returnvalue_4 false 0 :!class /Script/Engine.MeshComponent)
            (bind _asmaterial_instance_dynamic_2 (Utilities|Casting|CastToMaterialInstanceDynamic (Rendering|Material|GetOverlayMaterial _returnvalue_3 false 0 :!class /Script/Engine.MeshComponent))
              (:then
                (Rendering|Material|SetScalarParameterValue _asmaterial_instance_dynamic_2 "UseCenteredLines" 0.0 :!class /Script/Engine.MaterialInstanceDynamic) ; Fixes Thickness Issue in Front End
              )
              (:CastFailed
                (Development|PrintString "SOMETHING WENT WRONG APPLYING A DYNAMIC MATERIAL TO OVERLAY SLOT, THIS SHOULD NEVER BE SEEN IN THE LOG!")))))
        (Rendering|Material|SetOverlayMaterialMaxDrawDistance _returnvalue_3 1e+04 :!class /Script/Engine.MeshComponent)))))

(event EventOnDayPhaseChanged
  (params (CurrentDayPhase :type EFortDayPhase%20Enum :default "Morning") (PreviousDayPhase :type EFortDayPhase%20Enum :default "Morning") (bAtCreation :type Boolean)))

(event EventOnNewPlayerWeaponType
  (params (NewType :type EFortItemType%20Enum :default "WorldItem") (OldType :type EFortItemType%20Enum :default "WorldItem")))

(event EventOnWeaponChanged
  (params (NewWeapon :type Fort%20Weapon%20Object%20Reference) (OldWeapon :type Fort%20Weapon%20Object%20Reference)))

(event EventOnWeaponFired)

(event EventOnNewWeapons
  (params (WeaponsAsItems :type Array%20of%20Fort%20Item%20Object%20References)))

(event EventOnCurieStateAttached
  (params (Tag :type Gameplay%20Tag%20Structure%20%28by%20ref%29)))

(event EventOnCurieStateDetached
  (params (Tag :type Gameplay%20Tag%20Structure%20%28by%20ref%29)))

(event EventOnBeginSkydiving)

(event EventOnEndSkydiving)

(event EventOnEnterVehicle
  (params (VehicleEntered :type Fort%20Vehicle%20Interface%20Interface%20%28by%20ref%29)))

(event EventOnExitVehicle)

(event EventOnTargetingOrFirstPersonCameraChanged)

(event EventOnHitPawn
  (params (SelfActor :type Actor%20Object%20Reference) (OtherActor :type Actor%20Object%20Reference) (NormalImpulse :type Vector :default "0, 0, 0") (Hit :type Hit%20Result%20Structure%20%28by%20ref%29)))

(fn ConstructionScript)
