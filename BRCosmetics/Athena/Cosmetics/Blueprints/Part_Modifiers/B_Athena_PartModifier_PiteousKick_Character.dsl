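; Character part modifier for the PiteousKick cosmetic that makes its idle VFX react to the player's kills.
; After res-in, if reactive and outside the front end, it watches the AthenaKills stat once and stores increases in Kill Count.
; Kills one to three each play a dissolve timeline driving that tier's Global Alpha particle parameter, then step TransformProgress in thirds after a delay.
; Each tier also prints a debug string.
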(event EventOnPostResIn
  (bind _returnvalue (CallFunction|GetIdleVFXComponent))
  (|CallParentFunction)
  (if (Variables|Default|GetIs%20Reactive)
    (if (not (CustomCosmeticModifier|IsinFrontEnd))
      (Utilities|FlowControl|DoOnce false
        (:Completed
          (bind (_statname _statvalue) (Cosmetics|WatchCosmeticStat (CustomCharacter|GetAssociatedPlayerPawn) "(TagName=\"StatManager.AthenaKills\")" 0 false)
            (:then
              _)
            (:OnStatChanged
              (if (> _statvalue (Variables|Default|GetKill%20Count))
                _)))
          (bind _output_get (Variables|Default|SetKill%20Count _statvalue))
          (switch int _output_get
            (:1
              (Development|PrintString "Kill 1" true true "(R=0.807249,G=1.000000,B=0.000000,A=1.000000)" 6.0)
              (bind (_direction _dissolve) (Timeline|Dissolve01 PlayFromStart)
                (:Update
                  (Effects|Components|ParticleSystem|SetFloatParameter _returnvalue "Global Alpha Kill1" _dissolve)
                  (Utilities|FlowControl|Delay 1.0)
                  (Effects|Components|ParticleSystem|SetFloatParameter _returnvalue "TransformProgress" 0.33))))
            (:2
              (Development|PrintString "Kill 2" true true "(R=1.000000,G=0.607181,B=0.000000,A=1.000000)" 6.0)
              (bind (_direction_2 _dissolve_2) (Timeline|Dissolve02 Play)
                (:Update
                  (Effects|Components|ParticleSystem|SetFloatParameter _returnvalue "Global Alpha Kill2" _dissolve_2)
                  (Utilities|FlowControl|Delay 1.5)
                  (Effects|Components|ParticleSystem|SetFloatParameter _returnvalue "TransformProgress" 0.66))))
            (:3
              (Development|PrintString "Kill 3" true true "(R=1.000000,G=0.100628,B=0.000000,A=1.000000)" 6.0)
              (bind (_direction_3 _dissolve_3) (Timeline|Dissolve03 Play)
                (:Update
                  (Effects|Components|ParticleSystem|SetFloatParameter _returnvalue "Global Alpha Kill3" _dissolve_3)
                  (Utilities|FlowControl|Delay 1.0)
                  (Effects|Components|ParticleSystem|SetFloatParameter _returnvalue "TransformProgress" 1.0))))))))))

(event EventOnCharacterCustomizationFinished
  (params (CharacterPartInfo :type Array%20of%20Applied%20Character%20Part%20Info%20Structures)))

(fn ConstructionScript
  (|CallParentFunction))
