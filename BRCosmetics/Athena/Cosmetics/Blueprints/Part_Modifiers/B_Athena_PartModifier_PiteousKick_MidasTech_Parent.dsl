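; Parent character part modifier for the PiteousKick MidasTech cosmetic, adding kill-driven "super levelling" materials and a forced Midas-touch wrap on weapons and vehicles.
; Watches the AthenaKills stat to push level and change-time scalars to mesh materials, with a timer-based refresh workaround.
; In the front end, timelines loop through material levels and idle VFX dissolves.
; New weapons and entered vehicles get a wrap chosen by the Progressive variant stage.
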(event Custom|HackyWorkaroundRefreshTimerGuy
  ;[1 1 1] SuperLevelling Logic
  (Utilities|FlowControl|Sequence
    (:then_0
      (CallFunction|InitializeVariabels)
      (Rendering|Material|SetScalarParameterValueonMaterials (CustomCharacter|GetSkeletalMeshForPartOnPlayerPawn) "SL_CurrentLevel" (Math|Conversions|ToFloat(Integer) (stale EventOnPostResIn _output_get)))
      (Utilities|FlowControl|DoOnce false
        (:Completed
          (bind _returnvalue (Utilities|Time|SetTimerbyEvent (delegate Custom|HackyWorkaroundRefreshTimerGuy) 0.1 true))
          (Variables|Default|SetSL_RefreshTimeHandler _returnvalue))))
    (:then_1
      (Utilities|FlowControl|Gate true
        (:Exit
          (if (and (> (Math|Integer|Max(Integer) (stale EventOnPostResIn _statvalue) (Variables|Default|GetSL_LastValue)) (Variables|Default|GetSL_StartLevel)) (< (Math|Integer|Max(Integer) (stale EventOnPostResIn _statvalue) (Variables|Default|GetSL_LastValue)) (+ (Variables|Default|GetSL_EndLevel) 1)))
            (CallFunction|OnActualStatChange)))))))

(event Custom|ResetSafetyGate
  ;[1 1 1] SuperLevelling Logic
  (Utilities|FlowControl|DoOnce false
    (:Completed
      (bind _returnvalue (Utilities|Time|SetTimerbyEvent (delegate Custom|HackyWorkaroundRefreshTimerGuy) 0.1 true))
      (Variables|Default|SetSL_RefreshTimeHandler _returnvalue))))

(event EventOnPostResIn
  (bind _returnvalue (CallFunction|GetIdleVFXComponent))
  (bind _returnvalue_2 (CustomCharacter|GetSkeletalMeshForPartOnPlayerPawn))
  (bind _sl_lastvalue (Variables|Default|GetSL_LastValue))
  (|CallParentFunction)
  (CallFunction|ResetSafetyGate)
  ;[1 1 1] SuperLevelling Logic
  (Utilities|FlowControl|Sequence
    (:then_0
      (Utilities|FlowControl|Delay 1.0)
      (Utilities|Time|ClearandInvalidateTimerbyHandle (Variables|Default|GetSL_RefreshTimeHandler)))
    (:then_1
      (bind (_statname _statvalue) (Cosmetics|WatchCosmeticStat (CustomCharacter|GetAssociatedPlayerPawn) "(TagName=\"StatManager.AthenaKills\")")
        (:OnStatChanged
          (bind _output_get (Variables|Default|SetSL_LastValue (Math|Integer|Max(Integer) _statvalue _sl_lastvalue)))
          (Utilities|FlowControl|Sequence
            (:then_0
              (CallFunction|InitializeVariabels)
              (Rendering|Material|SetScalarParameterValueonMaterials (CustomCharacter|GetSkeletalMeshForPartOnPlayerPawn) "SL_CurrentLevel" (Math|Conversions|ToFloat(Integer) _output_get))
              (Utilities|FlowControl|DoOnce false
                (:Completed
                  (bind _returnvalue_3 (Utilities|Time|SetTimerbyEvent (delegate Custom|HackyWorkaroundRefreshTimerGuy) 0.1 true))
                  (Variables|Default|SetSL_RefreshTimeHandler _returnvalue_3))))
            (:then_1
              (Utilities|FlowControl|Gate true
                (:Exit
                  (if (and (> (Math|Integer|Max(Integer) _statvalue _sl_lastvalue) (Variables|Default|GetSL_StartLevel)) (< (Math|Integer|Max(Integer) _statvalue _sl_lastvalue) (+ (Variables|Default|GetSL_EndLevel) 1)))
                    (CallFunction|OnActualStatChange)))))))))
    (:then_2
      (if (CustomCosmeticModifier|IsinFrontEnd)
        (Timeline|Timeline_0 PlayFromStart
          (:Stage%202
            (Rendering|Material|SetScalarParameterValueonMaterials _returnvalue_2 "SL_CurrentLevel" 1.0)
            (Utilities|FlowControl|Sequence
              (:then_0
                ;[1 1 1] FX loop in FE
                (bind (_direction _dissolve_l1) (Timeline|Dissolve_FX_L1 PlayFromStart)
                  (:Update
                    (Effects|Components|ParticleSystem|SetFloatParameter _returnvalue "Global Alpha Kill1" _dissolve_l1))))
              (:then_1
                _)))
          (:Stage%203
            (Rendering|Material|SetScalarParameterValueonMaterials _returnvalue_2 "SL_CurrentLevel" 2.0)
            (Utilities|FlowControl|Sequence
              (:then_0
                (bind (_direction_2 _dissolve_l2) (Timeline|Dissolve_FX_L2 PlayFromStart)
                  (:Update
                    (Effects|Components|ParticleSystem|SetFloatParameter _returnvalue "Global Alpha Kill2" _dissolve_l2))))
              (:then_1
                _)))
          (:Stage%204
            (Rendering|Material|SetScalarParameterValueonMaterials _returnvalue_2 "SL_CurrentLevel" 3.0)
            (Utilities|FlowControl|Sequence
              (:then_0
                (bind (_direction_3 _dissolve_l3) (Timeline|Dissolve_FX_L3 PlayFromStart)
                  (:Update
                    (Effects|Components|ParticleSystem|SetFloatParameter _returnvalue "Global Alpha Kill3" _dissolve_l3))))
              (:then_1
                _))))
        (Rendering|Material|SetScalarParameterValueonMaterials (CustomCharacter|GetSkeletalMeshForPartOnPlayerPawn) "SL_GameTimeFromBP" (Utilities|Time|GetGameTimeinSeconds))))))

(event Custom|OnActualStatChange
  ;[1 1 1] SuperLevelling Logic
  ;;[1 1 1] This should only fire off when the stat changes - use this event!
  (Rendering|Material|SetScalarParameterValueonMaterials (CustomCharacter|GetSkeletalMeshForPartOnPlayerPawn) "SL_GameTimeFromBP" (Utilities|Time|GetGameTimeinSeconds)))

(event EventBeginPlay)

(event Custom|InitializeVariabels
  (bind _returnvalue (CustomCharacter|GetSkeletalMeshForPartOnPlayerPawn))
  (Rendering|Material|SetScalarParameterValueonMaterials _returnvalue "SL_StartAtLevel" (Math|Conversions|ToFloat(Integer) (Variables|Default|GetSL_StartLevel)))
  (Rendering|Material|SetScalarParameterValueonMaterials _returnvalue "SL_StopAtLevel" (Math|Conversions|ToFloat(Integer) (Variables|Default|GetSL_EndLevel)))
  (Utilities|NotwhenBID)
  (Rendering|Material|SetScalarParameterValueonMaterials _returnvalue "SL_EndRadius" (Variables|Default|GetEndRadius)))

(event EventOnNewWeapons
  (params (WeaponsAsItems :type Array%20of%20Fort%20Item%20Object%20References))
  (bind _pendingitemstowrap (Variables|Default|GetPendingItemsToWrap))
  ;[1 1 1] Force Weapons And Vehicles To Have A Specific Wrap (MIDAS TOUCH)
  (bind _index (Class|FortCosmeticFunctionLibrary|GetVariantChannelActiveIndexForSlot "(TagName=\"LoadoutSlot.Character\")" (CosmeticLoadout|GetCosmeticLoadout) "(TagName=\"Cosmetics.Variant.Channel.Progressive\")" (Utilities|Array|MakeArray "(TagName=\"Cosmetics.Variant.Property.Stage0\")" "(TagName=\"Cosmetics.Variant.Property.Stage1\")")))
  (if (== _index 1)
    (Variables|Default|SetWeaponWrapToUse (Variables|Default|GetWeaponWrapToUse_Pine))
    (else
      (Variables|Default|SetWeaponWrapToUse (Variables|Default|GetWeaponWrapToUse_Default))))
  (Utilities|NotwhenBID)
  (for _array_element WeaponsAsItems
    (switch Utilities|FlowControl|Switch|SwitchonEFortItemType (Item|GetType _array_element)
      (:WeaponMelee
        _)
      (:WeaponRanged
        _))
    (Utilities|Array|Add (Variables|Default|GetPendingItemsToWrap) _array_element))
  (if (Utilities|IsValidSoftObjectReference (Variables|Default|GetWeaponWrapToUse))
    (CallFunction|ApplyOverridetoItems _pendingitemstowrap)
    (Utilities|Array|Clear _pendingitemstowrap)))

(event EventOnEnterVehicle
  (params (VehicleEntered :type Fort%20Vehicle%20Interface%20Interface%20%28by%20ref%29))
  ;[1 1 1] Force Weapons And Vehicles To Have A Specific Wrap (MIDAS TOUCH)
  (bind _index (Class|FortCosmeticFunctionLibrary|GetVariantChannelActiveIndexForSlot "(TagName=\"LoadoutSlot.Character\")" (CosmeticLoadout|GetCosmeticLoadout) "(TagName=\"Cosmetics.Variant.Channel.Progressive\")" (Utilities|Array|MakeArray "(TagName=\"Cosmetics.Variant.Property.Stage0\")" "(TagName=\"Cosmetics.Variant.Property.Stage1\")")))
  (if (== _index 1)
    (Variables|Default|SetWeaponWrapToUse (Variables|Default|GetWeaponWrapToUse_Pine))
    (else
      (Variables|Default|SetWeaponWrapToUse (Variables|Default|GetWeaponWrapToUse_Default))))
  (Utilities|NotwhenBID)
  (if (Utilities|IsValidSoftObjectReference (Variables|Default|GetWeaponWrapToUse))
    (bind _asfort_player_controller (CallFunction|GetAssociatedPlayerController))
    (Utilities|IsValid _asfort_player_controller
      (:Is%20Valid
        (Class|FortPlayerController|ApplyOverrideWraptoVehicle _asfort_player_controller (Variables|Default|GetWeaponWrapToUse))))))

(fn ApplyOverrideToItems
  (params (Array :type Array%20of%20Fort%20Item%20Object%20References))
  (bind _asfort_player_controller (CallFunction|GetAssociatedPlayerController))
  (Utilities|IsValid _asfort_player_controller
    (:Is%20Valid
      (for _array_element Array
        (Class|FortPlayerController|ServerApplyOverrideWraptoItem _asfort_player_controller (Item|GetItemGuid _array_element) (Variables|Default|GetWeaponWrapToUse))))))

(fn CheckIfReactive
  (bind _returnvalue (CustomCharacter|GetAssociatedPlayerPawn))
  (bind _cosmeticloadout (Variables|/Script/FortniteGame.FortPlayerPawn|GetCosmeticLoadout _returnvalue))
  (if (Utilities|Enum|Equal(Enum) (CustomCharacter|GetCharacterPartType) "Backpack")
    ;[1 1 1] Backpack
    (bind _index (Class|FortCosmeticFunctionLibrary|%28Deprecated%29GetVariantChannelActiveIndexForSlot "(TagName=\"LoadoutSlot.Backpack\")" _cosmeticloadout "(TagName=\"Cosmetics.Variant.Channel.Material\")" (Utilities|Array|MakeArray "(TagName=\"Cosmetics.Variant.Property.Mat1\")" "(TagName=\"Cosmetics.Variant.Property.Mat2\")" "(TagName=\"Cosmetics.Variant.Property.Mat3\")")))
    (switch int _index
      (:0
        (Variables|Default|SetbReactive true))
      (:1
        _)
      (:2
        _))
    (Variables|Default|SetbReactive false)
    (else
      (bind _index_2 (Class|FortCosmeticFunctionLibrary|%28Deprecated%29GetVariantChannelActiveIndexForSlot "(TagName=\"LoadoutSlot.Character\")" _cosmeticloadout "(TagName=\"Cosmetics.Variant.Channel.Emissive\")" (Utilities|Array|MakeArray "(TagName=\"Cosmetics.Variant.Property.Emissive1\")" "(TagName=\"Cosmetics.Variant.Property.Emissive2\")")))
      (switch int _index_2
        (:0
          (Variables|Default|SetbReactive true))
        (:1
          (Variables|Default|SetbReactive false))))))

(fn GetAssociatedPlayerController
  (bind _returnvalue (CustomCharacter|GetAssociatedPlayerPawn))
  (Utilities|FlowControl|Sequence
    (:then_0
      (bind _aspawn (Utilities|Casting|CastToPawn (Vehicle|GetVehicleActor _returnvalue :!class /Script/FortniteGame.FortPlayerPawn)))
      (bind _asfort_player_controller (Utilities|Casting|CastToFortPlayerController (Pawn|GetController _aspawn)))
      (return _asfort_player_controller))
    (:then_1
      (bind _asfort_player_controller_2 (Utilities|Casting|CastToFortPlayerController (Pawn|GetController _returnvalue)))
      (return _asfort_player_controller_2))))

(fn ConstructionScript
  (bind _returnvalue (CustomCharacter|GetSkeletalMeshForPartOnPlayerPawn))
  (|CallParentFunction)
  (Rendering|Material|SetScalarParameterValueonMaterials (CustomCharacter|GetSkeletalMeshForPartOnPlayerPawn) "SL_GameTimeFromBP" -99999.0)
  (Rendering|Material|SetScalarParameterValueonMaterials _returnvalue "SL_StartAtLevel" (Math|Conversions|ToFloat(Integer) (Variables|Default|GetSL_StartLevel)))
  (Rendering|Material|SetScalarParameterValueonMaterials _returnvalue "SL_StopAtLevel" (Math|Conversions|ToFloat(Integer) (Variables|Default|GetSL_EndLevel))))
