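; Character part modifier for the Polished Jade cosmetic that drives its glow materials and idle VFX.
; In frontend, it toggles glow from Emissive3/4 variant tags, flags the VFX as frontend and binds to the Mind AnimBP's trigger event.
; That event's curve value feeds VFX IdleGlow and the glow material parameter, squared on this part but linear on the face.
; Outside frontend it only caches the player pawn; a separate event flags low-LOD VFX.
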(event EventOnStatChanged
  (params (State%20Name :type Name :default "None") (Stat%20Value :type Integer))
  ; GlowON_OFF FrontEnd
  (|CallParentFunction State%20Name Stat%20Value))

(event EventOnPostResIn
  (bind _returnvalue (CustomCharacter|GetSkeletalMeshForPartOnPlayerPawn))
  (bind _returnvalue_2 (CustomCharacter|GetAssociatedPlayerPawn))
  ;[0 0 0 0.5] Set Player Ref
  (|CallParentFunction)
  (if (Reactive|Frontend|IsCosmeticObjectDisplayedinFrontendContext 0)
    (bind _asf_med_polished_jade_mind_anim_bp (Utilities|Casting|CastToF_MED_PolishedJade_Mind_AnimBP (Components|SkeletalMesh|GetAnimInstance (Customization|GetSkeletalMeshforPartType _returnvalue_2 (CustomCharacter|GetCharacterPartType) :!class /Script/FortniteGame.FortPlayerPawn))))
    (Utilities|FlowControl|Sequence
      (:then_0
        ; GlowON_OFF FrontEnd
        (if (GameplayTags|HasTag (Cosmetics|GatherMetaTags (CustomCharacter|GetAssociatedPlayerPawn) "/BRCosmetics/Athena/Items/Cosmetics/Characters/Character_PolishedJade_Mind.Character_PolishedJade_Mind" :!class /Script/FortniteGame.FortPlayerPawn) "(TagName=\"Cosmetics.Variant.Property.Emissive3\")")
          (Rendering|Material|SetScalarParameterValueonMaterials _returnvalue "GlowPolishedJade_OnOff" 1.0)
          (elif (GameplayTags|HasTag (Cosmetics|GatherMetaTags (CustomCharacter|GetAssociatedPlayerPawn) "/BRCosmetics/Athena/Items/Cosmetics/Characters/Character_PolishedJade_Mind.Character_PolishedJade_Mind" :!class /Script/FortniteGame.FortPlayerPawn) "(TagName=\"Cosmetics.Variant.Property.Emissive4\")")
            (Rendering|Material|SetScalarParameterValueonMaterials _returnvalue "GlowPolishedJade_OnOff" 0.0))))
      (:then_1
        (Default|BindEventtoTriggerFX _asf_med_polished_jade_mind_anim_bp (delegate Custom|Trigger))
        ;[0 0 0 0.5] Set FrontEnd Bool
        (Effects|Components|ParticleSystem|SetBoolParameter (CallFunction|GetIdleVFXComponent) "bIsFrontEnd" true)))
    (else
      (bind _asfort_player_pawn_athena (Utilities|Casting|CastToFortPlayerPawnAthena _returnvalue_2))
      (Variables|Default|SetPlayerRef _asfort_player_pawn_athena))))

(event Custom|Trigger
  (params (Glow%20Curve :type Float%20%28double-precision%29))
  (bind _returnvalue (CustomCharacter|GetAssociatedPlayerPawn))
  (bind _returnvalue_2 (FortPawn|GetModifierforPartType _returnvalue "Face"))
  (Effects|Components|ParticleSystem|SetFloatParameter (CallFunction|GetIdleVFXComponent) "IdleGlow" Glow%20Curve)
  ; Glow Material
  ;
  (Rendering|Material|SetScalarParameterValueonMaterials (CustomCharacter|GetSkeletalMeshForPartOnPlayerPawn) "GlowPolishedJade" (* Glow%20Curve Glow%20Curve))
  (if (Utilities|IsValid _returnvalue_2 :!class /Script/Engine.KismetSystemLibrary)
    (Rendering|Material|SetScalarParameterValueonMaterials (CustomCharacter|GetSkeletalMeshForPartOnPlayerPawn _returnvalue_2) "GlowPolishedJade" Glow%20Curve)))

(event Custom|Event%20LOD%20Check
  ;[0 0 0 0.5] Low LOD
  (bind _asskeletal_mesh_component (Utilities|Casting|CastToSkeletalMeshComponent (CustomCharacter|GetSkeletalMeshForPartOnPlayerPawn)))
  (Effects|Components|ParticleSystem|SetBoolParameter (CallFunction|GetIdleVFXComponent) "bisLowLOD" (>= (Components|SkinnedMesh|GetPredictedLODLevel _asskeletal_mesh_component) 2)))

(fn ConstructionScript
  (|CallParentFunction))
