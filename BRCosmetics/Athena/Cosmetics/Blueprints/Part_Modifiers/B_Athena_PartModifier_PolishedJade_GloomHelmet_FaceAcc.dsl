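; Cosmetic part modifier Blueprint for the PolishedJade Gloom helmet face accessory, switching its glow to match the selected emissive variant.
; After post res-in and a frontend display check, it reads meta tags for the PolishedJade Mind character on the player pawn.
; Emissive3 sets GlowPolishedJade_OnOff to on and Emissive4 to off across the part's skeletal mesh materials.
; The stat-changed event calls the parent and has an empty frontend branch.
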(event EventOnStatChanged
  (params (State%20Name :type Name :default "None") (Stat%20Value :type Integer))
  (|CallParentFunction State%20Name Stat%20Value)
  (if (Reactive|Frontend|IsCosmeticObjectDisplayedinFrontendContext 0)
    _))

(event EventOnPostResIn
  (bind _returnvalue (CustomCharacter|GetSkeletalMeshForPartOnPlayerPawn))
  ;[0 0 0 0.5] Set Player Ref
  (|CallParentFunction)
  (if (Reactive|Frontend|IsCosmeticObjectDisplayedinFrontendContext 0)
    (if (GameplayTags|HasTag (Cosmetics|GatherMetaTags (CustomCharacter|GetAssociatedPlayerPawn) "/BRCosmetics/Athena/Items/Cosmetics/Characters/Character_PolishedJade_Mind.Character_PolishedJade_Mind" :!class /Script/FortniteGame.FortPlayerPawn) "(TagName=\"Cosmetics.Variant.Property.Emissive3\")")
      (Rendering|Material|SetScalarParameterValueonMaterials _returnvalue "GlowPolishedJade_OnOff" 1.0)
      (elif (GameplayTags|HasTag (Cosmetics|GatherMetaTags (CustomCharacter|GetAssociatedPlayerPawn) "/BRCosmetics/Athena/Items/Cosmetics/Characters/Character_PolishedJade_Mind.Character_PolishedJade_Mind" :!class /Script/FortniteGame.FortPlayerPawn) "(TagName=\"Cosmetics.Variant.Property.Emissive4\")")
        (Rendering|Material|SetScalarParameterValueonMaterials _returnvalue "GlowPolishedJade_OnOff" 0.0)))))

(fn ConstructionScript
  (|CallParentFunction))
