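; Part modifier for the Polygon character cosmetic that escalates its look through four stages (0-3) as cosmetic states activate.
; Init creates body and face MIDs, and in the front end starts a looping timer that previews stages, repeating 1-3 after the first pass.
; Stages set material parameters; 1-3 drive idle VFX level and bind Niagara to the face mesh.
; Stage 3 adds body lightning with hidden-part checks, which lower stages deactivate.
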(event EventOnPostResIn
  (bind _returnvalue (CustomCharacter|GetAssociatedPlayerPawn))
  (bind _returnvalue_2 (Customization|GetSkeletalMeshforPartType _returnvalue "Face" :!class /Script/FortniteGame.FortPlayerPawn))
  (|CallParentFunction)
  (Utilities|FlowControl|Sequence
    (:then_0
      ;[1 1 1] MID Setup
      (bind _returnvalue_3 (Rendering|Material|CreateDynamicMaterialInstance (CustomCharacter|GetSkeletalMeshForPartOnPlayerPawn) 0 (Rendering|Material|GetMaterial (CustomCharacter|GetSkeletalMeshForPartOnPlayerPawn) 0) "Polygon_Body_MID" :!class /Script/Engine.PrimitiveComponent))
      (Variables|Default|SetBodyMID _returnvalue_3)
      (Utilities|IsValid _returnvalue_2
        (:Is%20Valid
          (bind _returnvalue_4 (Rendering|Material|CreateDynamicMaterialInstance _returnvalue_2 0 (Rendering|Material|GetMaterial _returnvalue_2 0) "Polygon_Face_MID" :!class /Script/Engine.PrimitiveComponent))
          (Variables|Default|SetFaceMID _returnvalue_4))))
    (:then_1
      ;[1 1 1] Front End Preview
      (if (CustomCosmeticModifier|IsinFrontEnd)
        (Utilities|Time|SetTimerbyEvent (EventDispatchers|CreateEvent :!event Custom|FrontEndTrigger) 3.0 true)))))

(event Custom|FrontEndTrigger
  ;[1 1 1] Front End Preview
  (switch int (Variables|Default|GetFrontEndSwitch)
    (:0
      (CallFunction|Stage0))
    (:1
      (CallFunction|Stage1))
    (:2
      (CallFunction|Stage2))
    (:3
      (CallFunction|Stage3)
      (Variables|Default|SetFrontEndSwitch 0)))
  (Variables|Default|SetFrontEndSwitch (+ (Variables|Default|GetFrontEndSwitch) 1)))

(event EventOnCosmeticStateActivated
  (params (Stat%20Name :type Name :default "None") (Stat%20Value :type Integer) (STAGE%20Num%20Being%20Activated :type Integer))
  (|CallParentFunction Stat%20Name Stat%20Value STAGE%20Num%20Being%20Activated)
  (bind _output_get (Variables|Default|SetStatChange STAGE%20Num%20Being%20Activated))
  (switch int _output_get
    (:0
      (CallFunction|Stage0))
    (:1
      (CallFunction|Stage1))
    (:2
      (CallFunction|Stage2))
    (:3
      (CallFunction|Stage3))))

(event Custom|Stage%200
  (bind _character_lightning (Variables|Default|GetCharacter_Lightning))
  ;[1 1 1] Material and VFX Setup for Stages
  (Rendering|Material|SetScalarParameterValue (Variables|Default|GetBodyMID) "SuperLevel" 0.0 :!class /Script/Engine.MaterialInstanceDynamic)
  (Utilities|IsValid _character_lightning
    (:Is%20Valid
      (Components|Activation|Deactivate _character_lightning))))

(event Custom|Stage%201
  (bind _returnvalue (CallFunction|GetIdleVFXComponent))
  (bind _character_lightning (Variables|Default|GetCharacter_Lightning))
  ;[1 1 1] Material and VFX Setup for Stages
  (Rendering|Material|SetScalarParameterValue (Variables|Default|GetBodyMID) "SuperLevel" 1.0 :!class /Script/Engine.MaterialInstanceDynamic)
  (Rendering|Material|SetScalarParameterValue (Variables|Default|GetFaceMID) "SuperLevelAnim" (Utilities|Time|GetGameTimeinSeconds) :!class /Script/Engine.MaterialInstanceDynamic)
  (Components|Activation|Activate _returnvalue true)
  (Effects|Components|ParticleSystem|SetIntParameter _returnvalue "Level" 0)
  (Niagara|SetNiagaraSkeletalMeshComponent (Utilities|Casting|CastToNiagaraComponent _returnvalue) "SkeletalMesh_Face" (Customization|GetSkeletalMeshforPartType (CustomCharacter|GetAssociatedPlayerPawn) "Face" :!class /Script/FortniteGame.FortPlayerPawn))
  (Utilities|IsValid _character_lightning
    (:Is%20Valid
      (Components|Activation|Deactivate _character_lightning))))

(event Custom|Stage%202
  (bind _returnvalue (CallFunction|GetIdleVFXComponent))
  (bind _character_lightning (Variables|Default|GetCharacter_Lightning))
  ;[1 1 1] Material and VFX Setup for Stages
  (Rendering|Material|SetScalarParameterValue (Variables|Default|GetBodyMID) "SuperLevel" 2.0 :!class /Script/Engine.MaterialInstanceDynamic)
  (Rendering|Material|SetScalarParameterValue (Variables|Default|GetFaceMID) "SuperLevelAnim" (Utilities|Time|GetGameTimeinSeconds) :!class /Script/Engine.MaterialInstanceDynamic)
  (Components|Activation|Activate _returnvalue true)
  (Effects|Components|ParticleSystem|SetIntParameter _returnvalue "Level" 1)
  (Niagara|SetNiagaraSkeletalMeshComponent (Utilities|Casting|CastToNiagaraComponent _returnvalue) "SkeletalMesh_Face" (Customization|GetSkeletalMeshforPartType (CustomCharacter|GetAssociatedPlayerPawn) "Face" :!class /Script/FortniteGame.FortPlayerPawn))
  (Utilities|IsValid _character_lightning
    (:Is%20Valid
      (Components|Activation|Deactivate _character_lightning))))

(event Custom|Stage%203
  (bind _returnvalue (CallFunction|GetIdleVFXComponent))
  ;[1 1 1] Material and VFX Setup for Stages
  (Rendering|Material|SetScalarParameterValue (Variables|Default|GetBodyMID) "SuperLevel" 3.0 :!class /Script/Engine.MaterialInstanceDynamic)
  (Rendering|Material|SetScalarParameterValue (Variables|Default|GetFaceMID) "SuperLevelAnim" (Utilities|Time|GetGameTimeinSeconds) :!class /Script/Engine.MaterialInstanceDynamic)
  (Components|Activation|Activate _returnvalue true)
  (Effects|Components|ParticleSystem|SetIntParameter _returnvalue "Level" 2)
  (Niagara|SetNiagaraSkeletalMeshComponent (Utilities|Casting|CastToNiagaraComponent _returnvalue) "SkeletalMesh_Face" (Customization|GetSkeletalMeshforPartType (CustomCharacter|GetAssociatedPlayerPawn) "Face" :!class /Script/FortniteGame.FortPlayerPawn))
  (bind _returnvalue_2 (PartModifier|SpawnSystem|SpawnSystemAttachedPartModifier "/BRCosmetics/Characters/Player/Male/Medium/Bodies/M_MED_Polygon/FX/NS_Polygon_Body_Lightning.NS_Polygon_Body_Lightning" (CustomCharacter|GetSkeletalMeshForPartOnPlayerPawn) "root")
    (:Success
      (bind _output_get (Variables|Default|SetCharacter_Lightning _returnvalue_2))
      (Utilities|Array|Add (Variables|HiddenPartCheck|GetAdditional%20FX%20Needing%20Part%20Checks) _output_get)
      (Niagara|SetNiagaraSkeletalMeshComponent _output_get "SkeletalMesh_Face" (Customization|GetSkeletalMeshforPartType (CustomCharacter|GetAssociatedPlayerPawn) "Face" :!class /Script/FortniteGame.FortPlayerPawn))
      (Niagara|SetNiagaraVariable(Bool) _output_get "bisHeadHidden" (Variables|HiddenPartCheck|GetbIsHeadHidden))
      (Niagara|SetNiagaraVariable(Bool) _output_get "bisHandHidden" (Variables|HiddenPartCheck|GetbIsHandHidden))
      (Niagara|SetNiagaraVariable(Bool) _output_get "bisFootHidden" (Variables|HiddenPartCheck|GetbIsFootHidden)))))

(fn ConstructionScript
  (|CallParentFunction))
