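; Cosmetic part modifier for the PrismParticle face accessory that adjusts its materials and idle VFX when the player is downed (DBNO) or recovers.
; After res-in it sets distance-fade UV scaling by front-end state, binding the pawn's DBNO event only outside the front end.
; Going down stops glitch particle spawning, toggles a hide-element parameter and plays a GlitchProgress timeline.
; Recovery restores spawning, reverses the timeline and resets the element after a delay.
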(event EventOnPostResIn
  (bind _returnvalue (CustomCharacter|GetSkeletalMeshForPartOnPlayerPawn))
  (bind _material_instance_dynamic (Class|BTechArtGlobalLibrary|GetDynamicMaterialInstance _returnvalue))
  (|CallParentFunction)
  (if (CustomCosmeticModifier|IsinFrontEnd)
    (Rendering|Material|SetScalarParameterValue _material_instance_dynamic "PP_UseDistanceFadeUVScale" 1.0 :!class /Script/Engine.MaterialInstanceDynamic)
    (else
      (EventDispatchers|BindEventtoOnDBNOStateChanged (CustomCharacter|GetAssociatedPlayerPawn) (delegate Custom|DBNO))
      (Rendering|Material|SetScalarParameterValue _material_instance_dynamic "PP_UseDistanceFadeUVScale" 0.0 :!class /Script/Engine.MaterialInstanceDynamic))))

(event Custom|DBNO
  (params (FortPawn :type Fort%20Pawn%20Object%20Reference) (bInIsDBNO :type Boolean))
  (bind _returnvalue (CustomCharacter|GetAssociatedPlayerPawn))
  (bind _returnvalue_2 (Customization|GetSkeletalMeshforPartType _returnvalue "Face" :!class /Script/FortniteGame.FortPlayerPawn))
  (bind _material_instance_dynamic (Class|BTechArtGlobalLibrary|GetDynamicMaterialInstance _returnvalue_2 1))
  (Effects|Components|ParticleSystem|SetFloatParameter (CallFunction|GetIdleVFXComponent) "GlitchSpawnProbability" (Math|Conversions|ToFloat(Boolean) (not bInIsDBNO)))
  (if bInIsDBNO
    (Rendering|Material|SetScalarParameterValue _material_instance_dynamic "Hide Element 01" 0.0 :!class /Script/Engine.MaterialInstanceDynamic)
    (bind (_direction _newtrack_0) (Timeline|Timeline PlayFromStart)
      (:Update
        (Rendering|Material|SetScalarParameterValue (Class|BTechArtGlobalLibrary|GetDynamicMaterialInstance _returnvalue_2) "GlitchProgress" _newtrack_0 :!class /Script/Engine.MaterialInstanceDynamic)))
    (else
      (Utilities|FlowControl|Sequence
        (:then_0
          (Timeline|Timeline ReverseFromEnd))
        (:then_1
          (Utilities|FlowControl|Delay 1.3)
          (Rendering|Material|SetScalarParameterValue _material_instance_dynamic "Hide Element 01" 1.0 :!class /Script/Engine.MaterialInstanceDynamic))))))

(fn ConstructionScript
  (|CallParentFunction))
