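; Body part modifier for the Athena ProfPup cosmetic that drives a sprint state on its idle VFX.
; On each character movement update it checks whether the associated player pawn is sprinting outside the front end.
; DoOnce gates limit writes to the User.IsSprinting particle parameter, which is set true on sprint and cleared after a short delay, forming a brief pulse.
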(event EventOnCharacterMovementUpdated
  (params (DeltaSeconds :type Float%20%28single-precision%29) (OldLocation :type Vector :default "0, 0, 0") (OldVelocity :type Vector :default "0, 0, 0"))
  (bind _returnvalue (CallFunction|GetIdleVFXComponent))
  (|CallParentFunction DeltaSeconds OldLocation OldVelocity)
  (if (and (PawnStats|IsSprinting (CustomCharacter|GetAssociatedPlayerPawn)) (not (CustomCosmeticModifier|IsinFrontEnd)))
    (Utilities|FlowControl|DoOnce false
      (:Completed
        (Effects|Components|ParticleSystem|SetBoolParameter _returnvalue "User.IsSprinting" true)
        (Utilities|FlowControl|Delay 0.5)
        (Effects|Components|ParticleSystem|SetBoolParameter _returnvalue "User.IsSprinting" false)))
    (else
      _))
  (Utilities|FlowControl|DoOnce false
    (:Completed
      (Effects|Components|ParticleSystem|SetBoolParameter _returnvalue "User.IsSprinting" false))))

(fn ConstructionScript
  (|CallParentFunction))
