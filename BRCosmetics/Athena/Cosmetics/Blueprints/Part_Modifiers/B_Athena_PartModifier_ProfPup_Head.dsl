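; Cosmetic part modifier for the Professor Pup head that drives its motorcycle light, glow and idle VFX.
; After spawn it attaches the light to the FX_Light socket and binds to the anim Blueprint's lights event.
; That event sets the VFX IsWalkingBackwards parameter through DoOnce gates with no reset shown, plus debug prints.
; At Night a timeline raises light alpha and head emissive and flags NightPhase; other phases zero or clear them.
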(event EventOnPostResIn
  (|CallParentFunction)
  (Transformation|AttachComponentToComponent (Variables|Default|GetMotorcycle_Light) (CustomCharacter|GetSkeletalMeshForPartOnPlayerPawn) "FX_Light" "KeepRelative" "SnapToTarget" "KeepRelative" true)
  (bind _asm_lrg_professor_pup_anim_bp (Utilities|Casting|CastToM_LRG_Professor_Pup_AnimBP (Components|SkeletalMesh|GetAnimInstance (Customization|GetSkeletalMeshforPartType (CustomCharacter|GetAssociatedPlayerPawn) (CustomCharacter|GetCharacterPartType) :!class /Script/FortniteGame.FortPlayerPawn))))
  (Default|BindEventtoLights _asm_lrg_professor_pup_anim_bp (EventDispatchers|CreateEvent :!event Custom|Lights%20OnOff)))

(event Custom|Lights%20OnOff
  (params (BrakeLight :type Boolean))
  (bind _returnvalue (CallFunction|GetIdleVFXComponent))
  (if BrakeLight
    (Utilities|FlowControl|DoOnce false
      (:Completed
        (Development|PrintString "LIGHT _ TRUE")
        (Effects|Components|ParticleSystem|SetBoolParameter _returnvalue "IsWalkingBackwards" true))))
  (Utilities|FlowControl|DoOnce false
    (:Completed
      (Development|PrintString "LIGHT _ FALSE")
      (Effects|Components|ParticleSystem|SetBoolParameter _returnvalue "IsWalkingBackwards" false))))

(event EventOnDayPhaseChanged
  (params (CurrentDayPhase :type EFortDayPhase%20Enum :default "Morning") (PreviousDayPhase :type EFortDayPhase%20Enum :default "Morning") (bAtCreation :type Boolean))
  (|CallParentFunction CurrentDayPhase PreviousDayPhase bAtCreation)
  (if (Utilities|Enum|Equal(Enum) CurrentDayPhase "Night")
    (bind (_direction _lightalpha) (Timeline|TimeOfDayChanges PlayFromStart)
      (:Update
        (Rendering|Material|SetScalarParameterValueonMaterials (Variables|Default|GetMotorcycle_Light) "Alpha" _lightalpha)
        (Effects|Components|ParticleSystem|SetBoolParameter (CallFunction|GetIdleVFXComponent) "User.NightPhase" true)
        (Rendering|Material|SetScalarParameterValueonMaterials (CustomCharacter|GetSkeletalMeshForPartOnPlayerPawn) "Emissive_BrightnessMax" (* _lightalpha 5e+01))))
    (else
      (Rendering|Material|SetScalarParameterValueonMaterials (Variables|Default|GetMotorcycle_Light) "Alpha" 0.0)
      (Effects|Components|ParticleSystem|SetBoolParameter (CallFunction|GetIdleVFXComponent) "User.NightPhase" false)
      (Rendering|Material|SetScalarParameterValueonMaterials (CustomCharacter|GetSkeletalMeshForPartOnPlayerPawn) "Emissive_BrightnessMax" 0.0))))

(fn ConstructionScript
  (|CallParentFunction))
