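; Body part modifier for the Psyche cosmetic character that adds an idle goop effect to the player.
; After res-in it runs the parent logic, then spawns the goop Niagara system attached to the root of the part's skeletal mesh and stores it.
; It also enables custom depth on the idle VFX component, copying the stencil value from the pawn's mesh.
; Before destruction it deactivates the goop effect if still valid.
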(event EventOnPostResIn
  (bind _returnvalue (CallFunction|GetIdleVFXComponent))
  (|CallParentFunction)
  (bind _returnvalue_2 (PartModifier|SpawnSystem|SpawnSystemAttachedPartModifier "/BRCosmetics/Characters/Player/Male/Medium/Bodies/M_MED_Psyche/FX/NS_Psyche_Goop_Idle.NS_Psyche_Goop_Idle" (CustomCharacter|GetSkeletalMeshForPartOnPlayerPawn) "root" "0, 0, 0" "0, 0, 0" "SnapToTarget")
    (:Success
      (Variables|Default|SetGoopVFX _returnvalue_2))
    (:Failure
      _))
  (Rendering|SetRenderCustomDepth _returnvalue true)
  (Rendering|SetCustomDepthStencilValue _returnvalue (Variables|/Script/Engine.PrimitiveComponent|GetCustomDepthStencilValue (Variables|/Script/Engine.Character|GetMesh (CustomCharacter|GetAssociatedPlayerPawn)))))

(event EventPreDestruction
  (bind _goopvfx (Variables|Default|GetGoopVFX))
  (Utilities|IsValid _goopvfx
    (:Is%20Valid
      (Components|Activation|Deactivate _goopvfx))))

(fn ConstructionScript)
