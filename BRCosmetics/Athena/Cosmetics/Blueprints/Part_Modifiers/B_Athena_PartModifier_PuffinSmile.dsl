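; Cosmetic part modifier Blueprint for the PuffinSmile character part, adjusting its rim lighting for the front end.
; After the post-res-in event runs the parent logic, it checks whether the character is in the front end.
; If so, it sets the RimLight_IsFE scalar parameter on the part's skeletal mesh materials to enable the front-end rim light variant.
; The construction script only calls the parent.
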(event EventOnPostResIn
  (|CallParentFunction)
  (if (CustomCosmeticModifier|IsinFrontEnd)
    (Rendering|Material|SetScalarParameterValueonMaterials (CustomCharacter|GetSkeletalMeshForPartOnPlayerPawn) "RimLight_IsFE" 1.0)))

(fn ConstructionScript
  (|CallParentFunction))
