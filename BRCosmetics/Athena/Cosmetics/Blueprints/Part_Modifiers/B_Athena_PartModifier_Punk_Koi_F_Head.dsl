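; Cosmetic part modifier for the female Punk Koi character head that sets its idle VFX to match the progressive style stage.
; On post-res-in it calls the parent, waits briefly through a retriggerable delay, then reads the pawn's active Progressive variant index from the character loadout slot.
; The idle particle component is deactivated for the first two stages and activated for later ones, with the HeadGlowisActive parameter enabled only from the fourth.
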(event EventOnPostResIn
  (bind _returnvalue (CallFunction|GetIdleVFXComponent))
  (|CallParentFunction)
  (Utilities|FlowControl|RetriggerableDelay 0.05)
  (bind _index (CosmeticLoadout|GetVariantChannelActiveIndexforSlotActor "(TagName=\"LoadoutSlot.Character\")" (CustomCharacter|GetAssociatedPlayerPawn) "(TagName=\"Cosmetics.Variant.Channel.Progressive\")" (Utilities|Array|MakeArray "(TagName=\"Cosmetics.Variant.Property.Stage1\")" "(TagName=\"Cosmetics.Variant.Property.Stage2\")" "(TagName=\"Cosmetics.Variant.Property.Stage3\")" "(TagName=\"Cosmetics.Variant.Property.Stage4\")")))
  (if (< _index 2)
    (Components|Activation|Deactivate _returnvalue)
    (else
      (Components|Activation|Activate _returnvalue false)
      (if (< _index 3)
        (Effects|Components|ParticleSystem|SetBoolParameter _returnvalue "HeadGlowisActive" false)
        (else
          (Effects|Components|ParticleSystem|SetBoolParameter _returnvalue "HeadGlowisActive" true))))))

(fn ConstructionScript
  (|CallParentFunction))
