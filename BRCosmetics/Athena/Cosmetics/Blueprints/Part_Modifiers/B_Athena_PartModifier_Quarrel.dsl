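; Athena part modifier for the Quarrel cosmetic, likely a back bling, that makes its idle VFX react to kills, damage dealt and damage received.
; On post-res-in it matches the pawn mesh's custom depth stencil, resets counters and starts damage stat watchers; frontend previews play animations via timeline.
; Stat increases drive particle parameters: row offset plus game-time stamps for kills and damage dealt, timeline-driven glitch for damage received.
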(event EventOnDayPhaseChanged
  (params (CurrentDayPhase :type EFortDayPhase%20Enum :default "Morning") (PreviousDayPhase :type EFortDayPhase%20Enum :default "Morning") (bAtCreation :type Boolean)))

(event EventOnNewPlayerWeaponType
  (params (NewType :type EFortItemType%20Enum :default "WorldItem") (OldType :type EFortItemType%20Enum :default "WorldItem"))
  (|CallParentFunction NewType OldType) ; This node is disabled and will not be called.\nDrag off pins to build functionality.
)

(event EventOnWeaponChanged
  (params (NewWeapon :type Fort%20Weapon%20Object%20Reference) (OldWeapon :type Fort%20Weapon%20Object%20Reference)))

(event EventOnWeaponFired)

(event EventOnNewWeapons
  (params (WeaponsAsItems :type Array%20of%20Fort%20Item%20Object%20References)))

(event EventOnCurieStateAttached
  (params (Tag :type Gameplay%20Tag%20Structure%20%28by%20ref%29)))

(event EventOnCurieStateDetached
  (params (Tag :type Gameplay%20Tag%20Structure%20%28by%20ref%29)))

(event EventOnBeginSkydiving
  (|CallParentFunction) ; This node is disabled and will not be called.\nDrag off pins to build functionality.
)

(event EventOnEndSkydiving
  (|CallParentFunction) ; This node is disabled and will not be called.\nDrag off pins to build functionality.
)

(event EventOnEnterVehicle
  (params (VehicleEntered :type Fort%20Vehicle%20Interface%20Interface%20%28by%20ref%29)))

(event EventOnExitVehicle)

(event EventOnTargetingOrFirstPersonCameraChanged)

(event EventOnHitPawn
  (params (SelfActor :type Actor%20Object%20Reference) (OtherActor :type Actor%20Object%20Reference) (NormalImpulse :type Vector :default "0, 0, 0") (Hit :type Hit%20Result%20Structure%20%28by%20ref%29)))

(event EventOnPostResIn
  (bind _returnvalue (CallFunction|GetIdleVFXComponent))
  ;[1 1 1] Init
  (|CallParentFunction)
  (Rendering|SetRenderCustomDepth _returnvalue true)
  (Rendering|SetCustomDepthStencilValue _returnvalue (Variables|/Script/Engine.PrimitiveComponent|GetCustomDepthStencilValue (Variables|/Script/Engine.Character|GetMesh (CustomCharacter|GetAssociatedPlayerPawn))))
  (Variables|Default|Setkills 0)
  (Variables|Default|SetdamageDealt 0)
  (Variables|Default|SetdamageReceived 0)
  (CallFunction|OnDamageDealt)
  (CallFunction|OnDamageReceived)
  (if (Reactive|Frontend|ShouldDisplayFrontendReactivity 0 "Backpack")
    (Timeline|Timeline_0 PlayFromStart
      (:Stage%201
        (CallFunction|PlayDamageReceivedAnim))
      (:Stage%202
        (CallFunction|PlayKillAnim)))))

(event Custom|PlayKillAnim
  (bind _returnvalue (CallFunction|GetIdleVFXComponent))
  ;[1 1 1] Play Kill Anim
  (Effects|Components|ParticleSystem|SetFloatParameter _returnvalue "row_offset" 1.0)
  (Effects|Components|ParticleSystem|SetFloatParameter _returnvalue "kill_time" (Utilities|Time|GetGameTimeinSeconds)))

(event EventOnStatChanged
  (params (State%20Name :type Name :default "None") (Stat%20Value :type Integer))
  ;[1 1 1] On Kill
  (if (> Stat%20Value (Variables|Default|Getkills))
    (Variables|Default|Setkills Stat%20Value)
    (CallFunction|PlayKillAnim)))

(event Custom|OnDamageDealt
  (bind _returnvalue (CallFunction|GetIdleVFXComponent))
  ;[1 1 1] On Damage Dealt
  (bind (_statname _statvalue) (Cosmetics|WatchCosmeticStat (CustomCharacter|GetAssociatedPlayerPawn) "(TagName=\"StatManager.DamageDealt\")" 0 false)
    (:OnStatChanged
      (if (> _statvalue (Variables|Default|GetdamageDealt))
        (Variables|Default|SetdamageDealt _statvalue)
        (Effects|Components|ParticleSystem|SetFloatParameter _returnvalue "row_offset" 0.0)
        (Effects|Components|ParticleSystem|SetFloatParameter _returnvalue "damage_time" (Utilities|Time|GetGameTimeinSeconds))))))

(event Custom|OnDamageReceived
  ;[1 1 1] On Damage Received
  (bind (_statname _statvalue) (Cosmetics|WatchCosmeticStat (CustomCharacter|GetAssociatedPlayerPawn) "(TagName=\"StatManager.DamageReceived\")" 0 false)
    (:OnStatChanged
      (if (> _statvalue (Variables|Default|GetdamageReceived))
        (Variables|Default|SetdamageReceived _statvalue)
        (CallFunction|PlayDamageReceivedAnim)))))

(event Custom|PlayDamageReceivedAnim
  ;[1 1 1] Play Damage Received Anim
  (bind (_direction _glitch) (Timeline|Timeline_1 PlayFromStart)
    (:Update
      (Effects|Components|ParticleSystem|SetFloatParameter (CallFunction|GetIdleVFXComponent) "glitch_amount" _glitch))))

(fn ConstructionScript
  (|CallParentFunction))
