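; Part modifier for the Quartz backpack cosmetic that drives its idle VFX in response to damage received by the wearer.
; When the pawn or part changes, it watches the DamageReceived cosmetic stat on the associated player pawn.
; On each stat change it prints a "DAMAGE!" debug string and passes the current game time to the idle particle system's GameTimeFromBP parameter, timestamping the hit for the effect.
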(event EventOnStatChanged
  (params (State%20Name :type Name :default "None") (Stat%20Value :type Integer))
  (|CallParentFunction State%20Name Stat%20Value)
  (Development|PrintString "DAMAGE!")
  (Effects|Components|ParticleSystem|SetFloatParameter (CallFunction|GetIdleVFXComponent) "GameTimeFromBP" (Utilities|Time|GetGameTimeinSeconds)))

(event EventOnPawnAndPartChanged
  (|CallParentFunction)
  (Cosmetics|WatchCosmeticStat (CustomCharacter|GetAssociatedPlayerPawn) "(TagName=\"StatManager.DamageReceived\")" 0 false))

(fn ConstructionScript
  (|CallParentFunction))
