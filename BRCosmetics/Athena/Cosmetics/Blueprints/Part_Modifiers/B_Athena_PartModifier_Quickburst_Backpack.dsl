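; Part modifier for the Quickburst backpack cosmetic that makes its idle VFX and nitro materials react to player movement.
; On post-init it passes the active Progressive variant index to the VFX, then caches the player pawn in game or flip-flops the reactive state after a delay in frontend.
; In-game Tick feeds sprint/slide and jump states to the VFX, with sprint/slide also switching material speed and brightness.
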(event EventOnPostResIn
  ;[0 0 0] post res
  (|CallParentFunction)
  (if (Reactive|Frontend|IsCosmeticObjectDisplayedinFrontendContext 0)
    (CallFunction|VariantSetup)
    (Utilities|FlowControl|Delay 1.0)
    (bind _isa (Utilities|FlowControl|FlipFlop)
      (:A
        _)
      (:B
        _))
    (Effects|Components|ParticleSystem|SetBoolParameter (CallFunction|GetIdleVFXComponent) "SprintReactive" _isa)
    (CallFunction|NitroMaterialReactivity _isa)
    (Utilities|FlowControl|Delay 2.0)
    (else
      (bind _asfort_player_pawn_athena (Utilities|Casting|CastToFortPlayerPawnAthena (CustomCharacter|GetAssociatedPlayerPawn)))
      (Variables|Default|SetplayerRef _asfort_player_pawn_athena)
      (CallFunction|VariantSetup))))

(event EventTick
  (params (DeltaSeconds :type Float%20%28single-precision%29))
  (bind _returnvalue (CallFunction|GetIdleVFXComponent))
  (bind _playerref (Variables|Default|GetplayerRef))
  (bind _bistacticalsprinting (Variables|/Script/FortniteGame.FortPlayerPawn|GetbIsTacticalSprinting _playerref))
  (bind _bissliding (Variables|/Script/FortniteGame.FortPawn|GetbIsSliding _playerref))
  (bind _returnvalue_2 (or _bissliding _bistacticalsprinting))
  ;[0 0 0] tick fx
  (if (not (Reactive|Frontend|IsCosmeticObjectDisplayedinFrontendContext 0))
    (Effects|Components|ParticleSystem|SetBoolParameter _returnvalue "SprintReactive" _returnvalue_2)
    (Effects|Components|ParticleSystem|SetBoolParameter _returnvalue "PlayerJumping" (Variables|/Script/Engine.Character|GetbWasJumping _playerref))
    (CallFunction|NitroMaterialReactivity _returnvalue_2)))

(fn NitroMaterialReactivity
  (params (Reactivityctive :type Boolean))
  (bind _returnvalue (CustomCharacter|GetAssociatedPlayerPawn))
  (bind _returnvalue_2 (Customization|GetSkeletalMeshforPartType _returnvalue "Backpack" :!class /Script/FortniteGame.FortPlayerPawn))
  ;[1 0.3286 0] NitroTube
  (if Reactivityctive
    (Rendering|Material|SetScalarParameterValueonMaterials _returnvalue_2 "GlobalSpeed" 2.0)
    (Rendering|Material|SetScalarParameterValueonMaterials _returnvalue_2 "NitroBrightness" 3.0)
    (Rendering|Material|SetScalarParameterValueonMaterials _returnvalue_2 "VolatileBrightness" 6.0)
    (else
      (Rendering|Material|SetScalarParameterValueonMaterials _returnvalue_2 "GlobalSpeed" 0.2)
      (Rendering|Material|SetScalarParameterValueonMaterials _returnvalue_2 "NitroBrightness" 1.1)
      (Rendering|Material|SetScalarParameterValueonMaterials _returnvalue_2 "VolatileBrightness" 3.0)))
  (return))

(fn ConstructionScript
  (|CallParentFunction))

(fn VariantSetup
  (bind _asfort_player_pawn (Utilities|Casting|CastToFortPlayerPawn (Actor|GetOwner)))
  (bind _index (CosmeticLoadout|GetVariantChannelActiveIndexforSlotActor "(TagName=\"LoadoutSlot.Backpack\")" _asfort_player_pawn "(TagName=\"Cosmetics.Variant.Channel.Progressive\")" (Utilities|Array|MakeArray "(TagName=\"Cosmetics.Variant.Property.Stage1\")" "(TagName=\"Cosmetics.Variant.Property.Stage2\")")))
  (Effects|Components|ParticleSystem|SetFloatParameter (CallFunction|GetIdleVFXComponent) "Variant" (Math|Conversions|ToFloat(Integer) _index)))
