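; RacerZero character part modifier that attaches a lotus flower mesh to the player and opens it at night, closing it otherwise.
; After the parent post-res-in, the lotus snaps to the part mesh's dyn_ball socket, falling back to the idle FX socket on skins lacking it.
; Day phase changes that flip the open state push it to lotus materials and idle VFX, with game time stamped as OpenTime on the materials.
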(event EventOnPostResIn
  (bind _returnvalue (CustomCharacter|GetSkeletalMeshForPartOnPlayerPawn))
  (bind _sm_racerzero_lotus (Variables|Default|GetSM_RacerZero_Lotus))
  (|CallParentFunction)
  ;[1 1 1] Added Fallback if the original fails, which put the flower at 0,0,0 and hid it on skins that dont have that socket, like scholar
  (if (Transformation|DoesSocketExist _returnvalue "dyn_ball")
    ;[1 1 1] Original
    (Transformation|AttachComponentToComponent _sm_racerzero_lotus _returnvalue "dyn_ball" "SnapToTarget" "KeepRelative" "KeepWorld" true)
    (else
      (Transformation|AttachComponentToComponent _sm_racerzero_lotus _returnvalue (CallFunction|GetIdleFXSocketName) "SnapToTarget" "KeepRelative" "KeepWorld" true))))

(event EventOnDayPhaseChanged
  (params (CurrentDayPhase :type EFortDayPhase%20Enum :default "Morning") (PreviousDayPhase :type EFortDayPhase%20Enum :default "Morning") (bAtCreation :type Boolean))
  (bind _floweropen (Variables|Default|GetFlowerOpen))
  (bind _returnvalue (Math|Conversions|ToFloat(Boolean) _floweropen))
  (bind _returnvalue_2 (Utilities|Enum|Equal(Enum) CurrentDayPhase "Night"))
  (bind _sm_racerzero_lotus (Variables|Default|GetSM_RacerZero_Lotus))
  (if (or (and _returnvalue_2 (not (Variables|Default|GetFlowerOpen))) (and (Variables|Default|GetFlowerOpen) (Utilities|Enum|NotEqual(Enum) CurrentDayPhase "Night")))
    (Variables|Default|SetFlowerOpen _returnvalue_2)
    (Rendering|Material|SetScalarParameterValueonMaterials _sm_racerzero_lotus "OpenTime" (Utilities|Time|GetGameTimeinSeconds))
    (Rendering|Material|SetScalarParameterValueonMaterials _sm_racerzero_lotus "Open" _returnvalue)
    (Effects|Components|ParticleSystem|SetFloatParameter (CallFunction|GetIdleVFXComponent) "Open" _returnvalue)))

(fn ConstructionScript
  (|CallParentFunction))
