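; Part modifier Blueprint for the Racer Zero Mask body cosmetic that manages a smoke particle component on the character.
; After post res-in, it snaps the smoke component to the root of the part's skeletal mesh on the player pawn.
; The smoke deactivates when skydiving begins and reactivates, with a reset, when skydiving ends.
; Each event calls its parent implementation first.
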(event EventOnPostResIn
  (|CallParentFunction)
  (Transformation|AttachComponentToComponent (Variables|Default|GetP_MED_Racer_Zero_Mask_ParticleSmoke_SpawnUnit) (CustomCharacter|GetSkeletalMeshForPartOnPlayerPawn) "root" "SnapToTarget" "SnapToTarget" "SnapToTarget" true))

(event EventOnEndSkydiving
  (|CallParentFunction)
  (Components|Activation|Activate (Variables|Default|GetP_MED_Racer_Zero_Mask_ParticleSmoke_SpawnUnit) true))

(event EventOnBeginSkydiving
  (|CallParentFunction)
  (Components|Activation|Deactivate (Variables|Default|GetP_MED_Racer_Zero_Mask_ParticleSmoke_SpawnUnit)))

(fn ConstructionScript)
