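; Cosmetic part modifier for RageDebris that makes a character part's material react to the player's kill count.
; On a stat change it stores the value as AthenaKillStat, then triggers reactivity if the value is above zero and outside the front end.
; Reactivity writes the count divided by ten, its ones digit and the current game time into the part mesh's custom primitive data as a vector at index 2.
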(event EventOnStatChanged
  (params (State%20Name :type Name :default "None") (Stat%20Value :type Integer))
  (|CallParentFunction State%20Name Stat%20Value)
  (bind _output_get (Variables|Default|SetAthenaKillStat Stat%20Value))
  (if (> _output_get 0)
    (if (not (CustomCosmeticModifier|IsinFrontEnd))
      (CallFunction|TriggerReactivity))))

(event Custom|TriggerReactivity
  (Rendering|Material|SetCustomPrimitiveDataVector3 (CustomCharacter|GetSkeletalMeshForPartOnPlayerPawn) 2 (Math|Vector|MakeVector (Math|Conversions|ToFloat(Integer) (/ (Variables|Default|GetAthenaKillStat) 10)) (Math|Conversions|ToFloat(Integer) (Math|Integer|%25%28Integer%29 (Variables|Default|GetAthenaKillStat) 10)) (Utilities|Time|GetGameTimeinSeconds))))

(fn ConstructionScript
  (|CallParentFunction))
