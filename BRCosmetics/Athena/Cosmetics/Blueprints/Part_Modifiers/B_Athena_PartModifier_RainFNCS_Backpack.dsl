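; Part modifier for the RainFNCS backpack cosmetic that makes its material react to the player's kills in game.
; After initialization outside the front end, it watches the AthenaKills stat on the associated player pawn.
; Each stat change restarts a timeline that drives the RainFNCS_Reactivity scalar on the part mesh's dynamic material.
; The front-end preview branch holds only chained delays, with no visible effect.
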(event EventOnPostResIn
  (|CallParentFunction)
  (if (CustomCosmeticModifier|IsinFrontEnd)
    ; FE preview cycle
    (Utilities|FlowControl|Delay 2.5)
    (Utilities|FlowControl|Sequence
      (:then_0
        (Utilities|FlowControl|Delay 7.5)
        (Utilities|FlowControl|Sequence
          (:then_0
            (Utilities|FlowControl|Delay 5.0)))))
    (else
      ; In game reactivity
      (Utilities|FlowControl|Sequence
        (:then_0
          (Cosmetics|WatchCosmeticStat (CustomCharacter|GetAssociatedPlayerPawn) "(TagName=\"StatManager.AthenaKills\")" 0 false
            (:OnStatChanged
              (bind (_direction _reactivity) (Timeline|RainFNCS_InGameReactivity PlayFromStart)
                (:Update
                  (Rendering|Material|SetScalarParameterValue (Class|BTechArtGlobalLibrary|GetDynamicMaterialInstance (CustomCharacter|GetSkeletalMeshForPartOnPlayerPawn)) "RainFNCS_Reactivity" _reactivity :!class /Script/Engine.MaterialInstanceDynamic))))))))))

(fn ConstructionScript
  (|CallParentFunction))
