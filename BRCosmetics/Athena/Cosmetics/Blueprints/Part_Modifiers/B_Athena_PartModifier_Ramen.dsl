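; Cosmetic part modifier Blueprint for the Ramen character part that adds a splash effect to the Ramen backpack when the player pawn lands.
; On landing, it spawns the backpack's JumpSplash Niagara system attached at the root of the skeletal mesh fetched for the part on the player pawn.
; The jump event is empty, while stat-change, post-res-in and construction logic only call the parent.
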(event EventOnStatChanged
  (params (State%20Name :type Name :default "None") (Stat%20Value :type Integer))
  (|CallParentFunction State%20Name Stat%20Value))

(event EventOnPostResIn
  (|CallParentFunction))

(event EventOnPawnLanded
  (Niagara|SpawnSystemAttached "/BRCosmetics/Accessories/FORT_Backpacks/Backpack_M_MED_Ramen/FX/NS_Backpack_Ramen_JumpSplash.NS_Backpack_Ramen_JumpSplash" (CustomCharacter|GetSkeletalMeshForPartOnPlayerPawn) "root" "0, 0, 0" "0, 0, 0" "SnapToTarget" true true "AutoRelease"))

(event EventOnPawnJumped)

(fn ConstructionScript
  (|CallParentFunction))
