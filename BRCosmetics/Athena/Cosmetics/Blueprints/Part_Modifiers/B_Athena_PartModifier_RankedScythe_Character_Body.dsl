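; Character body part modifier for the Ranked Scythe cosmetic that drives an idle effect from a ranked progression level.
; VFXUpdate activates the idle VFX component from level 16 upward and deactivates it below that.
; It maps upper levels to three RankedProgression tiers through a Select, enabling UseTriangle only at level 22.
; The KeepIdleAlwaysOn input is unused, and all event handlers are empty or disabled stubs.
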(event EventOnPostResIn
  (|CallParentFunction) ; This node is disabled and will not be called.\nDrag off pins to build functionality.
)

(event EventOnDayPhaseChanged
  (params (CurrentDayPhase :type EFortDayPhase%20Enum :default "Morning") (PreviousDayPhase :type EFortDayPhase%20Enum :default "Morning") (bAtCreation :type Boolean)))

(event EventOnNewPlayerWeaponType
  (params (NewType :type EFortItemType%20Enum :default "WorldItem") (OldType :type EFortItemType%20Enum :default "WorldItem"))
  (|CallParentFunction NewType OldType) ; This node is disabled and will not be called.\nDrag off pins to build functionality.
)

(event EventOnWeaponChanged
  (params (NewWeapon :type Fort%20Weapon%20Object%20Reference) (OldWeapon :type Fort%20Weapon%20Object%20Reference)))

(event EventOnWeaponFired)

(event EventOnNewWeapons
  (params (WeaponsAsItems :type Array%20of%20Fort%20Item%20Object%20References)))

(event EventOnCurieStateAttached
  (params (Tag :type Gameplay%20Tag%20Structure%20%28by%20ref%29)))

(event EventOnCurieStateDetached
  (params (Tag :type Gameplay%20Tag%20Structure%20%28by%20ref%29)))

(event EventOnBeginSkydiving
  (|CallParentFunction) ; This node is disabled and will not be called.\nDrag off pins to build functionality.
)

(event EventOnEndSkydiving
  (|CallParentFunction) ; This node is disabled and will not be called.\nDrag off pins to build functionality.
)

(event EventOnEnterVehicle
  (params (VehicleEntered :type Fort%20Vehicle%20Interface%20Interface%20%28by%20ref%29))
  (|CallParentFunction VehicleEntered) ; This node is disabled and will not be called.\nDrag off pins to build functionality.
)

(event EventOnExitVehicle
  (|CallParentFunction) ; This node is disabled and will not be called.\nDrag off pins to build functionality.
)

(event EventOnTargetingOrFirstPersonCameraChanged)

(event EventOnHitPawn
  (params (SelfActor :type Actor%20Object%20Reference) (OtherActor :type Actor%20Object%20Reference) (NormalImpulse :type Vector :default "0, 0, 0") (Hit :type Hit%20Result%20Structure%20%28by%20ref%29)))

(fn ConstructionScript
  (|CallParentFunction))

(fn VFXUpdate
  (params (Progression%20Level :type Integer) (KeepIdleAlwaysOn :type Boolean))
  (bind _returnvalue (CallFunction|GetIdleVFXComponent))
  (if (>= Progression%20Level 16)
    (Components|Activation|Activate _returnvalue false)
    (Effects|Components|ParticleSystem|SetIntParameter _returnvalue "RankedProgression" (Utilities|Select :Index Progression%20Level :Option%200 -1 :Option%201 -1 :Option%202 -1 :Option%203 -1 :Option%204 -1 :Option%205 -1 :Option%206 -1 :Option%207 -1 :Option%208 -1 :Option%209 -1 :Option%2010 -1 :Option%2011 -1 :Option%2012 -1 :Option%2013 -1 :Option%2014 -1 :Option%2015 -1 :Option%2016 0 :Option%2017 0 :Option%2018 0 :Option%2019 1 :Option%2020 1 :Option%2021 1 :Option%2022 2))
    (Effects|Components|ParticleSystem|SetBoolParameter _returnvalue "UseTriangle" (== Progression%20Level 22))
    (else
      (Components|Activation|Deactivate _returnvalue))))
