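; Body part modifier for the Ranked Trooper cosmetic that reflects player rank through tiered materials and idle Niagara VFX.
; A watched rank stat drives SetCurrentProgress; a chosen rank variant is capped by current rank, except in frontend or eligible warmup.
; Materials span Bronze to Unreal, while VFX progression updates only from Elite upward.
; VFX also receives style, mask, kills (randomized in preview) and timer-polled LOD and skin-weight profile.
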(event Custom|SetRank
  (params (Rank :type Integer))
  ;[0.4121 0.08369 1] This is a ke command to debug the asset rank in runtime. Just type: "ke * SetRank 0", being "0" the rank you want to debug.
  ;;[1 0.9115 0.5051] Only run if in Editor or Pie
  (if (Editor|InEditororPIE)
    (Default|BackpackMaterialSetup Rank)
    (CallFunction|VFXUpdate Rank)))

(event EventOnPawnAndPartChanged
  (bind _returnvalue (CustomCharacter|GetAssociatedPlayerPawn))
  (|CallParentFunction)
  (if (Variables|Rank|Debug|GetDEBUG_RankedProgression)
    (Default|SetCurrentProgress (Variables|Rank|Debug|GetDEBUG_RankedProgressionLevel))
    (else
      (Utilities|IsValid _returnvalue
        (:Is%20Valid
          (bind (_statname _statvalue) (Cosmetics|WatchCosmeticStat _returnvalue (Variables|Rank|GetRank%20Stat%20Tag))
            (:OnStatChanged
              (Default|SetCurrentProgress _statvalue))))))))

(event EventOnCosmeticPartAppliedPostResIn
  (params (Part :type Custom%20Character%20Part%20Object%20Reference))
  (Default|SetCurrentProgress (Variables|Rank|GetCurrentProgressionLevel)))

(event EventOnPostResIn
  (bind _returnvalue (CustomCharacter|GetAssociatedPlayerPawn))
  (bind _returnvalue_2 (Customization|GetSkeletalMeshforPartType _returnvalue "Body" :!class /Script/FortniteGame.FortPlayerPawn))
  (|CallParentFunction)
  (Utilities|FlowControl|Sequence
    (:then_0
      (Default|SetCurrentProgress (Variables|Rank|GetCurrentProgressionLevel)))
    (:then_1
      ;[0.2011 0.6562 0.000313 0.2261] Style variant (default, noble)
      (bind _index (CosmeticLoadout|GetVariantChannelActiveIndexforSlotActor "(TagName=\"LoadoutSlot.Character\")" (CustomCharacter|GetAssociatedPlayerPawn) "(TagName=\"Cosmetics.Variant.Channel.Parts\")" (Utilities|Array|MakeArray "(TagName=\"Cosmetics.Variant.Property.Stage1\")" "(TagName=\"Cosmetics.Variant.Property.Stage2\")") false false))
      (bind _asniagara_particle_system_component (Utilities|Casting|CastToNiagaraComponent (CallFunction|GetIdleVFXComponent)))
      (Niagara|SetNiagaraVariable(Int32) _asniagara_particle_system_component "StyleIndex" _index))
    (:then_2
      ;[0.2011 0.6562 0.000313 0.2261] Mask variant (mask on/mask off)
      (bind _index_2 (CosmeticLoadout|GetVariantChannelActiveIndexforSlotActor "(TagName=\"LoadoutSlot.Character\")" (CustomCharacter|GetAssociatedPlayerPawn) "(TagName=\"Cosmetics.Variant.Channel.Material\")" (Utilities|Array|MakeArray "(TagName=\"Cosmetics.Variant.Property.Mat1\")" "(TagName=\"Cosmetics.Variant.Property.Mat2\")") false false))
      (bind _asniagara_particle_system_component_2 (Utilities|Casting|CastToNiagaraComponent (CallFunction|GetIdleVFXComponent)))
      (Niagara|SetNiagaraVariable(Int32) _asniagara_particle_system_component_2 "MaskIndex" _index_2))
    (:then_3
      ;[1 0 0.617 0] Add mask to idle vfx component. Mask is always the same mesh, just hidden by material
      (bind _asniagara_particle_system_component_3 (Utilities|Casting|CastToNiagaraComponent (CallFunction|GetIdleVFXComponent)))
      (Niagara|SetNiagaraSkeletalMeshComponent _asniagara_particle_system_component_3 "MaskMesh" (Customization|GetSkeletalMeshforPartType (CustomCharacter|GetAssociatedPlayerPawn) "Face" :!class /Script/FortniteGame.FortPlayerPawn)))
    (:then_4
      (if (Reactive|Frontend|IsCosmeticObjectDisplayedasPreview (CustomCharacter|GetAssociatedPlayerPawn))
        ;[1 0.7991 0 0.5] Front end
        (Utilities|Time|SetTimerbyEvent (delegate Custom|Frontend_ReactiveBurst) 5.0 true true)
        (else
          ;[1 0.1856 0.1478 0.5] Kill reactivity
          (bind (_statname _statvalue) (Cosmetics|WatchCosmeticStat (CustomCharacter|GetAssociatedPlayerPawn) "(TagName=\"StatManager.AthenaKills\")" 0 false)
            (:OnStatChanged
              (if (> _statvalue 0)
                (bind _asniagara_particle_system_component_4 (Utilities|Casting|CastToNiagaraComponent (CallFunction|GetIdleVFXComponent)))
                (Niagara|SetNiagaraVariable(Int32) _asniagara_particle_system_component_4 "KillCount" _statvalue)))))))
    (:then_5
      (Utilities|Time|SetTimerbyEvent (EventDispatchers|CreateEvent :!event Custom|LODCheck) 0.2 true)
      (Variables|Default|SetBody%20Skeletal%20Mesh _returnvalue_2)
      (CallFunction|UpdateFXSkinWeights)
      (bind _ascustom_character_part_anim_instance (Utilities|Casting|CastToCustomCharacterPartAnimInstance (Components|SkeletalMesh|GetAnimInstance _returnvalue_2)))
      (Variables|Default|SetBody%20Anim%20Instance _ascustom_character_part_anim_instance))))

(event Custom|Frontend_ReactiveBurst
  ;[1 0.7991 0 0.5] Front end
  (bind _asniagara_particle_system_component (Utilities|Casting|CastToNiagaraComponent (CallFunction|GetIdleVFXComponent)))
  (Niagara|SetNiagaraVariable(Int32) _asniagara_particle_system_component "KillCount" (Math|Random|RandomInteger 99999)))

(event Custom|LODCheck
  (bind _body_anim_instance (Variables|Default|GetBody%20Anim%20Instance)
    (:then
      (Effects|Components|ParticleSystem|SetIntParameter (CallFunction|GetIdleVFXComponent) "SkeletalMeshCurrentLOD" (Variables|/Script/FortniteGame.CustomCharacterPartAnimInstance|GetCurrentLOD _body_anim_instance))
      (CallFunction|UpdateFXSkinWeights))
    (:else)))

(event Custom|UpdateFXSkinWeights
  (bind _body_skeletal_mesh (Variables|Default|GetBody%20Skeletal%20Mesh)
    (:then
      (switch name (Components|SkinnedMesh|GetCurrentSkinWeightProfileName _body_skeletal_mesh "Primary")
        (:Default
          ;[0.7552 0.2064 0 0.5] Check for skin weight profile changes
          (Utilities|FlowControl|DoOnce false
            (:Completed
              (Effects|Components|ParticleSystem|SetBoolParameter (CallFunction|GetIdleVFXComponent) "UseAlternateSkinWeights" false))))
        (:Alternative
          _))
      (Utilities|FlowControl|DoOnce false
        (:Completed
          (Effects|Components|ParticleSystem|SetBoolParameter (CallFunction|GetIdleVFXComponent) "UseAlternateSkinWeights" true))))
    (:else)))

(fn BackpackMaterialSetup
  (params (Index :type Integer))
  (bind _returnvalue (Rendering|Material|CreateDynamicMaterialInstance (Utilities|Select :Index Index :Option%200 (Variables|Rank|RankMaterials|GetBronze%20Material) :Option%201 (Variables|Rank|RankMaterials|GetBronze%20Material) :Option%202 (Variables|Rank|RankMaterials|GetBronze%20Material) :Option%203 (Variables|Rank|RankMaterials|GetBronze%20Material) :Option%204 (Variables|Rank|RankMaterials|GetSilver%20Material) :Option%205 (Variables|Rank|RankMaterials|GetSilver%20Material) :Option%206 (Variables|Rank|RankMaterials|GetSilver%20Material) :Option%207 (Variables|Rank|RankMaterials|GetGold%20Material) :Option%208 (Variables|Rank|RankMaterials|GetGold%20Material) :Option%209 (Variables|Rank|RankMaterials|GetGold%20Material) :Option%2010 (Variables|Rank|RankMaterials|GetPlatnium%20Material) :Option%2011 (Variables|Rank|RankMaterials|GetPlatnium%20Material) :Option%2012 (Variables|Rank|RankMaterials|GetPlatnium%20Material) :Option%2013 (Variables|Rank|RankMaterials|GetDiamond%20Material) :Option%2014 (Variables|Rank|RankMaterials|GetDiamond%20Material) :Option%2015 (Variables|Rank|RankMaterials|GetDiamond%20Material) :Option%2016 (Variables|Rank|RankMaterials|GetElite%20Material) :Option%2017 (Variables|Rank|RankMaterials|GetElite%20Material) :Option%2018 (Variables|Rank|RankMaterials|GetElite%20Material) :Option%2019 (Variables|Rank|RankMaterials|GetChampion%20Material) :Option%2020 (Variables|Rank|RankMaterials|GetChampion%20Material) :Option%2021 (Variables|Rank|RankMaterials|GetChampion%20Material) :Option%2022 (Variables|Rank|RankMaterials|GetUnreal%20Material)) "None" "None" :!class /Script/Engine.KismetMaterialLibrary))
  (bind _index (CosmeticLoadout|GetVariantChannelActiveIndexforSlotActor "(TagName=\"LoadoutSlot.Character\")" (CustomCharacter|GetAssociatedPlayerPawn) "(TagName=\"Cosmetics.Variant.Channel.Material\")" (Utilities|Array|MakeArray "(TagName=\"Cosmetics.Variant.Property.Mat0\")" "(TagName=\"Cosmetics.Variant.Property.Mat1\")") false false))
  (Rendering|Material|SetScalarParameterValue _returnvalue "Hide Element 01" (Math|Float|SelectFloat 1.0 0.0 (== _index 1)) :!class /Script/Engine.MaterialInstanceDynamic)
  (Rendering|Material|SetMaterial (CustomCharacter|GetSkeletalMeshForPartOnPlayerPawn) 0 _returnvalue))

(fn GetChosenHabaneroRank
  (bind _asfort_player_pawn (Utilities|Casting|CastToFortPlayerPawn (Actor|GetOwner)))
  (bind (_outvariantchannels _bfoundchannel) (Class|FortCosmeticFunctionLibrary|%28Deprecated%29GetVariantChannelInfoForSlotTag "(TagName=\"LoadoutSlot.Character\")" (Cosmetics|GetPawnCosmeticLoadout _asfort_player_pawn false true) "(TagName=\"Cosmetics.Variant.Channel.Ranked\")" false))
  (return _bfoundchannel (Class|FortCosmeticRankedVariant|GetHabaneroRankfromRankIndex (Utilities|String|StringToInteger (.CustomData (Utilities|Struct|BreakMcpVariantChannelInfo _outvariantchannels))))))

(fn SetCurrentProgress
  (params (NewProgress :type Integer))
  (bind _returnvalue (CallFunction|ShouldDisplayChosenRank))
  (bind _chosenvarianthabanerorank (LocalVariables|GetChosenVariantHabaneroRank))
  (bind _chosenvarianthabanerorank_2 (LocalVariables|GetChosenVariantHabaneroRank))
  (bind _isusingvariants (LocalVariables|GetIsUsingVariants))
  (bind _currentprogressionlevel (Variables|Rank|GetCurrentProgressionLevel))
  (bind (_is_using_variants _chosen_habanero_rank) (CallFunction|GetChosenHabaneroRank))
  (LocalVariables|SetIsUsingVariants _is_using_variants)
  (LocalVariables|SetChosenVariantHabaneroRank _chosen_habanero_rank)
  (if (or (LocalVariables|GetIsUsingVariants) (!= (Variables|GetNewProgress) (Variables|Rank|GetCurrentProgressionLevel)))
    (bind _output_get (Variables|Rank|SetCurrentProgressionLevel (Variables|GetNewProgress)))
    (Default|BackpackMaterialSetup (Math|Integer|SelectInt (Math|Integer|SelectInt _chosenvarianthabanerorank_2 (Math|Integer|Min(Integer) _output_get _chosenvarianthabanerorank) _returnvalue) _currentprogressionlevel _isusingvariants))
    (CallFunction|VFXUpdate (Math|Integer|SelectInt (Math|Integer|SelectInt _chosenvarianthabanerorank_2 (Math|Integer|Min(Integer) _output_get _chosenvarianthabanerorank) _returnvalue) _currentprogressionlevel _isusingvariants))))

(fn ShouldDisplayChosenRank
  (if (Reactive|Frontend|IsCosmeticObjectDisplayedinFrontendContext 0)
    (return true)
    (else
      (bind _output_get (LocalVariables|SetGamePhaseLogic (BattleRoyaleGamePhaseLogic|GetBattleRoyaleGamePhaseLogicComponent)))
      (Utilities|IsValid _output_get
        (:Is%20Valid
          (if (BattleRoyaleGamePhaseLogic|ShouldCosmeticUseGamePhaseBehaviour (LocalVariables|GetGamePhaseLogic))
            (return (Utilities|Enum|Equal(Enum) (BattleRoyaleGamePhaseLogic|GetGamePhase (LocalVariables|GetGamePhaseLogic)) "Warmup"))
            (else
              (return false))))
        (:Is%20Not%20Valid
          (return false))))))

(fn ConstructionScript
  (|CallParentFunction))

(fn VFXUpdate
  (params (Progression%20Level :type Integer))
  (if (not (< Progression%20Level 16))
    (Effects|Components|ParticleSystem|SetIntParameter (CallFunction|GetIdleVFXComponent) "RankedProgression" (- Progression%20Level 16))))
