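; Character part modifier for the Ranked Trooper head that drives its idle VFX from cosmetic variants, kills and progression.
; On post-init, it writes the active style and mask variant indices to Niagara variables.
; It sets KillCount from the watched kill stat, or randomizes it on a looping timer in frontend preview.
; SetCurrentProgress feeds VFXUpdate, which sets RankedProgression as an offset once progress reaches a threshold.
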(event EventOnPostResIn
  (|CallParentFunction)
  (Utilities|FlowControl|Sequence
    (:then_1
      ;[0.2011 0.6562 0.000313 0.2261] Style variant (default, noble)
      (bind _index (CosmeticLoadout|GetVariantChannelActiveIndexforSlotActor "(TagName=\"LoadoutSlot.Character\")" (CustomCharacter|GetAssociatedPlayerPawn) "(TagName=\"Cosmetics.Variant.Channel.Parts\")" (Utilities|Array|MakeArray "(TagName=\"Cosmetics.Variant.Property.Stage1\")" "(TagName=\"Cosmetics.Variant.Property.Stage2\")") false false))
      (bind _asniagara_particle_system_component (Utilities|Casting|CastToNiagaraComponent (CallFunction|GetIdleVFXComponent)))
      (Niagara|SetNiagaraVariable(Int32) _asniagara_particle_system_component "StyleIndex" _index))
    (:then_2
      ;[0.2011 0.6562 0.000313 0.2261] Mask variant (mask on/mask off)
      (bind _index_2 (CosmeticLoadout|GetVariantChannelActiveIndexforSlotActor "(TagName=\"LoadoutSlot.Character\")" (CustomCharacter|GetAssociatedPlayerPawn) "(TagName=\"Cosmetics.Variant.Channel.Material\")" (Utilities|Array|MakeArray "(TagName=\"Cosmetics.Variant.Property.Mat1\")" "(TagName=\"Cosmetics.Variant.Property.Mat2\")") false false))
      (bind _asniagara_particle_system_component_2 (Utilities|Casting|CastToNiagaraComponent (CallFunction|GetIdleVFXComponent)))
      (Niagara|SetNiagaraVariable(Int32) _asniagara_particle_system_component_2 "MaskIndex" _index_2))
    (:then_4
      (if (Reactive|Frontend|IsCosmeticObjectDisplayedasPreview (CustomCharacter|GetAssociatedPlayerPawn))
        ;[1 0.7991 0 0.5] Front end
        (Utilities|Time|SetTimerbyEvent (delegate Custom|Frontend_ReactiveBurst) 5.0 true true)
        (else
          ;[1 0.1856 0.1478 0.5] Kill reactivity
          (bind (_statname _statvalue) (Cosmetics|WatchCosmeticStat (CustomCharacter|GetAssociatedPlayerPawn) "(TagName=\"StatManager.AthenaKills\")" 0 false)
            (:OnStatChanged
              (if (> _statvalue 0)
                (bind _asniagara_particle_system_component_3 (Utilities|Casting|CastToNiagaraComponent (CallFunction|GetIdleVFXComponent)))
                (Niagara|SetNiagaraVariable(Int32) _asniagara_particle_system_component_3 "KillCount" _statvalue)))))))))

(event Custom|Frontend_ReactiveBurst
  ;[1 0.7991 0 0.5] Front end
  (bind _asniagara_particle_system_component (Utilities|Casting|CastToNiagaraComponent (CallFunction|GetIdleVFXComponent)))
  (Niagara|SetNiagaraVariable(Int32) _asniagara_particle_system_component "KillCount" (Math|Random|RandomInteger 99999)))

(fn SetCurrentProgress
  (params (NewProgress :type Integer))
  (CallFunction|VFXUpdate NewProgress))

(fn ConstructionScript
  (|CallParentFunction))

(fn VFXUpdate
  (params (Progression%20Level :type Integer))
  (if (not (< Progression%20Level 16))
    (Effects|Components|ParticleSystem|SetIntParameter (CallFunction|GetIdleVFXComponent) "RankedProgression" (- Progression%20Level 16))))
