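; Part modifier for the Ranked Trooper skin that drives its idle Niagara VFX.
; On post-resin init, it passes style and mask variant indices and the face mesh to the effect.
; A looping timer syncs body LOD and flags skin weight profile changes through DoOnce nodes.
; Watched kill stats set KillCount in matches, while front-end previews randomize it on a timer to fake reactivity.
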(event EventOnPostResIn
  (bind _returnvalue (CustomCharacter|GetAssociatedPlayerPawn))
  (bind _returnvalue_2 (Customization|GetSkeletalMeshforPartType _returnvalue "Body" :!class /Script/FortniteGame.FortPlayerPawn))
  (|CallParentFunction)
  (Utilities|FlowControl|Sequence
    (:then_0
      ;[0.2011 0.6562 0.000313 0.2261] Style variant (default, noble)
      (bind _index (CosmeticLoadout|GetVariantChannelActiveIndexforSlotActor "(TagName=\"LoadoutSlot.Character\")" (CustomCharacter|GetAssociatedPlayerPawn) "(TagName=\"Cosmetics.Variant.Channel.Parts\")" (Utilities|Array|MakeArray "(TagName=\"Cosmetics.Variant.Property.Stage1\")" "(TagName=\"Cosmetics.Variant.Property.Stage2\")") false false))
      (bind _asniagara_particle_system_component (Utilities|Casting|CastToNiagaraComponent (CallFunction|GetIdleVFXComponent)))
      (Niagara|SetNiagaraVariable(Int32) _asniagara_particle_system_component "StyleIndex" _index))
    (:then_1
      ;[0.2011 0.6562 0.000313 0.2261] Mask variant (mask on/mask off)
      (bind _index_2 (CosmeticLoadout|GetVariantChannelActiveIndexforSlotActor "(TagName=\"LoadoutSlot.Character\")" (CustomCharacter|GetAssociatedPlayerPawn) "(TagName=\"Cosmetics.Variant.Channel.Material\")" (Utilities|Array|MakeArray "(TagName=\"Cosmetics.Variant.Property.Mat1\")" "(TagName=\"Cosmetics.Variant.Property.Mat2\")") false false))
      (bind _asniagara_particle_system_component_2 (Utilities|Casting|CastToNiagaraComponent (CallFunction|GetIdleVFXComponent)))
      (Niagara|SetNiagaraVariable(Int32) _asniagara_particle_system_component_2 "MaskIndex" _index_2))
    (:then_2
      ;[1 0 0.617 0] Add mask to idle vfx component. Mask is always the same mesh, just hidden by material
      (bind _asniagara_particle_system_component_3 (Utilities|Casting|CastToNiagaraComponent (CallFunction|GetIdleVFXComponent)))
      (Niagara|SetNiagaraSkeletalMeshComponent _asniagara_particle_system_component_3 "MaskMesh" (Customization|GetSkeletalMeshforPartType (CustomCharacter|GetAssociatedPlayerPawn) "Face" :!class /Script/FortniteGame.FortPlayerPawn)))
    (:then_3
      (Utilities|Time|SetTimerbyEvent (EventDispatchers|CreateEvent :!event Custom|LODCheck) 0.2 true)
      (Variables|Default|SetBodySkeletalMesh _returnvalue_2)
      (CallFunction|UpdateFXSkinWeights)
      (bind _ascustom_character_part_anim_instance (Utilities|Casting|CastToCustomCharacterPartAnimInstance (Components|SkeletalMesh|GetAnimInstance _returnvalue_2)))
      (Variables|Default|SetBody%20Anim%20Instance _ascustom_character_part_anim_instance))
    (:then_4
      (if (Reactive|Frontend|IsCosmeticObjectDisplayedasPreview (CustomCharacter|GetAssociatedPlayerPawn))
        ;[1 0.7991 0 0.5] Front end
        (Utilities|Time|SetTimerbyEvent (delegate Custom|Frontend_ReactiveBurst) 5.0 true true)
        (else
          ;[1 0.1856 0.1478 0.5] Kill reactivity
          (bind (_statname _statvalue) (Cosmetics|WatchCosmeticStat (CustomCharacter|GetAssociatedPlayerPawn) "(TagName=\"StatManager.AthenaKills\")" 0 false)
            (:OnStatChanged
              (if (> _statvalue 0)
                (bind _asniagara_particle_system_component_4 (Utilities|Casting|CastToNiagaraComponent (CallFunction|GetIdleVFXComponent)))
                (Niagara|SetNiagaraVariable(Int32) _asniagara_particle_system_component_4 "KillCount" _statvalue)))))))))

(event Custom|Frontend_ReactiveBurst
  ;[1 0.7991 0 0.5] Front end
  (bind _asniagara_particle_system_component (Utilities|Casting|CastToNiagaraComponent (CallFunction|GetIdleVFXComponent)))
  (Niagara|SetNiagaraVariable(Int32) _asniagara_particle_system_component "KillCount" (Math|Random|RandomInteger 99999)))

(event Custom|LODCheck
  (bind _body_anim_instance (Variables|Default|GetBody%20Anim%20Instance)
    (:then
      (Effects|Components|ParticleSystem|SetIntParameter (CallFunction|GetIdleVFXComponent) "SkeletalMeshCurrentLOD" (Variables|/Script/FortniteGame.CustomCharacterPartAnimInstance|GetCurrentLOD _body_anim_instance))
      (CallFunction|UpdateFXSkinWeights))
    (:else)))

(event Custom|UpdateFXSkinWeights
  (bind _bodyskeletalmesh (Variables|Default|GetBodySkeletalMesh)
    (:then
      (switch name (Components|SkinnedMesh|GetCurrentSkinWeightProfileName _bodyskeletalmesh "Primary")
        (:Default
          ;[0.7552 0.2064 0 0.5] Check for skin weight profile changes
          (Utilities|FlowControl|DoOnce false
            (:Completed
              (Effects|Components|ParticleSystem|SetBoolParameter (CallFunction|GetIdleVFXComponent) "UseAlternateSkinWeights" false))))
        (:Alternative
          _))
      (Utilities|FlowControl|DoOnce false
        (:Completed
          (Effects|Components|ParticleSystem|SetBoolParameter (CallFunction|GetIdleVFXComponent) "UseAlternateSkinWeights" true))))
    (:else)))

(fn ConstructionScript
  (|CallParentFunction))
