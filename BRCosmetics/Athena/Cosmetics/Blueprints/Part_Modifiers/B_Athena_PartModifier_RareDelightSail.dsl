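; Cosmetic part modifier for the RareDelightSail character that adapts its idle VFX and materials to display state.
; On part application and weapon change, it toggles arm render parameters on the idle VFX, switching to the gauntlet variant when the hand is hidden.
; In the front end, a looping timer polls for hologram display and swaps two body material slots to a blank material, without restoring them.
; The timer is cleared on destroy.
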(event EventOnCosmeticPartAppliedPostResIn
  (params (Part :type Custom%20Character%20Part%20Object%20Reference))
  (bind _returnvalue (CallFunction|GetIdleVFXComponent))
  ;[0 0 0] Is Gauntlet check
  (|CallParentFunction Part)
  (if (Variables|HiddenPartCheck|GetbIsHandHidden)
    (Effects|Components|ParticleSystem|SetBoolParameter _returnvalue "RightArmRender" false)
    (Effects|Components|ParticleSystem|SetBoolParameter _returnvalue "LeftArmDefaultRender" false)
    (Effects|Components|ParticleSystem|SetBoolParameter _returnvalue "RightArmGauntletRender" true)
    (else
      (Effects|Components|ParticleSystem|SetBoolParameter _returnvalue "RightArmRender" true)
      (Effects|Components|ParticleSystem|SetBoolParameter _returnvalue "LeftArmDefaultRender" true)
      (Effects|Components|ParticleSystem|SetBoolParameter _returnvalue "RightArmGauntletRender" false))))

(event EventOnWeaponChanged
  (params (NewWeapon :type Fort%20Weapon%20Object%20Reference) (OldWeapon :type Fort%20Weapon%20Object%20Reference))
  (bind _returnvalue (CallFunction|GetIdleVFXComponent))
  ;[0 0 0] Is Gauntlet check
  (|CallParentFunction NewWeapon OldWeapon)
  (if (Variables|HiddenPartCheck|GetbIsHandHidden)
    (Effects|Components|ParticleSystem|SetBoolParameter _returnvalue "RightArmRender" false)
    (Effects|Components|ParticleSystem|SetBoolParameter _returnvalue "LeftArmDefaultRender" false)
    (Effects|Components|ParticleSystem|SetBoolParameter _returnvalue "RightArmGauntletRender" true)
    (else
      (Effects|Components|ParticleSystem|SetBoolParameter _returnvalue "RightArmRender" true)
      (Effects|Components|ParticleSystem|SetBoolParameter _returnvalue "LeftArmDefaultRender" true)
      (Effects|Components|ParticleSystem|SetBoolParameter _returnvalue "RightArmGauntletRender" false))))

(event EventOnPostResIn
  (|CallParentFunction)
  (Utilities|FlowControl|Sequence
    (:then_1
      (if (CustomCosmeticModifier|IsinFrontEnd)
        (bind _returnvalue (Utilities|Time|SetTimerbyEvent (delegate Custom|CE_StateTick) 0.2 true))
        (Variables|Default|SetTimer _returnvalue)))))

(event Custom|CE_StateTick
  (bind _returnvalue (CustomCharacter|GetSkeletalMeshForPartOnPlayerPawn))
  (if (Variables|/Script/FortniteGame.FortPlayerPawn|GetbIsInFrontEndHologram (CustomCharacter|GetAssociatedPlayerPawn))
    (bind _asskeletal_mesh_component (Utilities|Casting|CastToSkeletalMeshComponent _returnvalue))
    (if (== (Variables|/Script/Engine.SkinnedMeshComponent|GetSkeletalMesh _asskeletal_mesh_component) "/Game/Characters/Player/Female/Medium/Bodies/F_MED_RareDelightSail/Meshes/F_MED_RareDelightSail.F_MED_RareDelightSail")
      (Rendering|Material|SetMaterial _returnvalue 1 "/Game/Accessories/FORT_Backpacks/M_MED_Celestial_Backpack/M_Blank.M_Blank")
      (Rendering|Material|SetMaterial _returnvalue 2 "/Game/Accessories/FORT_Backpacks/M_MED_Celestial_Backpack/M_Blank.M_Blank"))))

(event EventDestroyed
  (|CallParentFunction)
  (Utilities|Time|ClearandInvalidateTimerbyHandle (Variables|Default|GetTimer)))

(fn ConstructionScript
  (|CallParentFunction))
