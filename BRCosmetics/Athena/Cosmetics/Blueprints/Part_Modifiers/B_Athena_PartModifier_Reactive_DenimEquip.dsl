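; Part modifier for the DenimEquip backpack that plays a reactive Niagara burst on eliminations.
; It watches the pawn's AthenaKills stat and, on a value above zero, prints a debug "ELIM", waits briefly, then spawns the attached effect with backpack variants applied.
; When frontend reactivity should display, a looping timer started once at PostResIn repeats the same burst and is cleared on destroy.
; StatValueToStage always returns 0.
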(event EventOnDayPhaseChanged
  (params (CurrentDayPhase :type EFortDayPhase%20Enum :default "Morning") (PreviousDayPhase :type EFortDayPhase%20Enum :default "Morning") (bAtCreation :type Boolean)))

(event EventOnNewPlayerWeaponType
  (params (NewType :type EFortItemType%20Enum :default "WorldItem") (OldType :type EFortItemType%20Enum :default "WorldItem"))
  (|CallParentFunction NewType OldType) ; This node is disabled and will not be called.\nDrag off pins to build functionality.
)

(event EventOnWeaponChanged
  (params (NewWeapon :type Fort%20Weapon%20Object%20Reference) (OldWeapon :type Fort%20Weapon%20Object%20Reference)))

(event EventOnWeaponFired)

(event EventOnNewWeapons
  (params (WeaponsAsItems :type Array%20of%20Fort%20Item%20Object%20References)))

(event EventOnCurieStateAttached
  (params (Tag :type Gameplay%20Tag%20Structure%20%28by%20ref%29)))

(event EventOnCurieStateDetached
  (params (Tag :type Gameplay%20Tag%20Structure%20%28by%20ref%29)))

(event EventOnBeginSkydiving
  (|CallParentFunction) ; This node is disabled and will not be called.\nDrag off pins to build functionality.
)

(event EventOnEndSkydiving
  (|CallParentFunction) ; This node is disabled and will not be called.\nDrag off pins to build functionality.
)

(event EventOnEnterVehicle
  (params (VehicleEntered :type Fort%20Vehicle%20Interface%20Interface%20%28by%20ref%29))
  (|CallParentFunction VehicleEntered) ; This node is disabled and will not be called.\nDrag off pins to build functionality.
)

(event EventOnExitVehicle
  (|CallParentFunction) ; This node is disabled and will not be called.\nDrag off pins to build functionality.
)

(event EventOnTargetingOrFirstPersonCameraChanged)

(event EventOnHitPawn
  (params (SelfActor :type Actor%20Object%20Reference) (OtherActor :type Actor%20Object%20Reference) (NormalImpulse :type Vector :default "0, 0, 0") (Hit :type Hit%20Result%20Structure%20%28by%20ref%29)))

(event EventOnStatChanged
  (params (State%20Name :type Name :default "None") (Stat%20Value :type Integer)))

(event EventOnPawnAndPartChanged
  (|CallParentFunction)
  (Cosmetics|WatchCosmeticStat (CustomCharacter|GetAssociatedPlayerPawn) "(TagName=\"StatManager.AthenaKills\")"
    (:OnStatChanged
      (if (> (stale EventOnStatChanged Stat%20Value) 0)
        (Development|PrintString "ELIM")
        (Utilities|FlowControl|Delay 0.5)
        (bind _returnvalue (PartModifier|SpawnSystem|SpawnSystemAttachedPartModifier "/BRCosmetics/Accessories/FORT_Backpacks/Backpack_F_MED_DenimEquip/FX/NS_DenimEquip_Backpack_ElimReactive.NS_DenimEquip_Backpack_ElimReactive" (CustomCharacter|GetSkeletalMeshForPartOnPlayerPawn))
          (:Success
            (Cosmetics|ApplyVariantsToComponent(Deprecated) :self (CosmeticLoadout|GetItemDefinitionfromLoadout (CosmeticLoadout|GetCosmeticLoadout) "(TagName=\"LoadoutSlot.Backpack\")") :InComponent _returnvalue :Loadout (Variables|/Script/FortniteGame.FortPlayerPawn|GetCosmeticLoadout (CustomCharacter|GetAssociatedPlayerPawn)))))))))

(event Custom|CE_Frontend%20FX%20Loop%20Burst
  (bind _returnvalue (PartModifier|SpawnSystem|SpawnSystemAttachedPartModifier "/BRCosmetics/Accessories/FORT_Backpacks/Backpack_F_MED_DenimEquip/FX/NS_DenimEquip_Backpack_ElimReactive.NS_DenimEquip_Backpack_ElimReactive" (CustomCharacter|GetSkeletalMeshForPartOnPlayerPawn))
    (:Success
      (Cosmetics|ApplyVariantsToComponent(Deprecated) :self (CosmeticLoadout|GetItemDefinitionfromLoadout (CosmeticLoadout|GetCosmeticLoadout) "(TagName=\"LoadoutSlot.Backpack\")") :InComponent _returnvalue :Loadout (Variables|/Script/FortniteGame.FortPlayerPawn|GetCosmeticLoadout (CustomCharacter|GetAssociatedPlayerPawn))))))

(event EventOnPostResIn
  (|CallParentFunction)
  (if (Reactive|Frontend|ShouldDisplayFrontendReactivity 0 "Backpack")
    (Utilities|FlowControl|DoOnce false
      (:Completed
        (bind _returnvalue (Utilities|Time|SetTimerbyEvent (delegate Custom|CE_Frontend%20FX%20Loop%20Burst) 6.0 true))
        (Variables|Default|SetTimer _returnvalue)))))

(event EventDestroyed
  (if (Reactive|Frontend|ShouldDisplayFrontendReactivity 0 "Backpack")
    (Utilities|Time|ClearandInvalidateTimerbyHandle (Variables|Default|GetTimer))))

(fn StatValueToStage
  (params (StatValue :type Integer))
  (return 0))

(fn ConstructionScript
  (|CallParentFunction))
