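; Part modifier for the RebelClaw backpack cosmetic that drives its idle VFX and spawns Niagara smoke bursts.
; On post-res-in it flags front-end preview, where a retriggerable delay precedes a burst, as in ReplayCountDown.
; A watched kill stat spawns smoke after a short delay and briefly zeroes idle spawn scale.
; The Progressive variant stage selects smoke and idle colors, and skydiving state sets idle interpolation parameters.
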(event Custom|ReplayCountDown
  ;[0.141 0.4403 1 1 200] FRONTEND 🛒
  (Utilities|FlowControl|RetriggerableDelay 3.0)
  (bind _returnvalue (PartModifier|SpawnSystem|SpawnSystemAttachedPartModifier "/BRCosmetics/Accessories/FORT_Backpacks/Backpack_F_MED_RebelClaw/FX/NS_Backpack_RebelClaw_Smoke.NS_Backpack_RebelClaw_Smoke" (CustomCharacter|GetSkeletalMeshForPartOnPlayerPawn) "Root")
    (:Success
      (if (or (Reactive|Frontend|IsCosmeticObjectDisplayedasPreview 0) (CustomCosmeticModifier|IsinFrontEnd))
        (Effects|Components|ParticleSystem|SetBoolParameter _returnvalue "bIsFrontEnd" true)
        (Effects|Components|ParticleSystem|SetBoolParameter _returnvalue "bIsFrontEndPreview" true)
        (bind _index (Class|FortCosmeticFunctionLibrary|GetVariantChannelActiveIndexForSlot "(TagName=\"LoadoutSlot.Backpack\")" (CosmeticLoadout|GetCosmeticLoadout) "(TagName=\"Cosmetics.Variant.Channel.Progressive\")" (Utilities|Array|MakeArray "(TagName=\"Cosmetics.Variant.Property.Stage1\")" "(TagName=\"Cosmetics.Variant.Property.Stage2\")")))
        (if (== _index 0)
          (Effects|Components|ParticleSystem|SetColorParameter _returnvalue "StartColor" "(R=0.109376,G=0.266993,B=0.300000,A=1.000000)")
          (Effects|Components|ParticleSystem|SetColorParameter _returnvalue "EndColor" "(R=0.250000,G=0.250000,B=0.250000,A=1.000000)"))
        (Effects|Components|ParticleSystem|SetColorParameter _returnvalue "StartColor" "(R=0.109376,G=0.266993,B=0.300000,A=1.000000)")
        (Effects|Components|ParticleSystem|SetColorParameter _returnvalue "EndColor" "(R=0.140454,G=0.102357,B=0.300000,A=1.000000)"))
      (Effects|Components|ParticleSystem|SetBoolParameter _returnvalue "bIsFrontEnd" false)
      (Effects|Components|ParticleSystem|SetBoolParameter _returnvalue "bIsFrontEndPreview" false))))

(event EventOnPostResIn
  (bind _returnvalue (CallFunction|GetIdleVFXComponent))
  (bind _returnvalue_2 (CallFunction|GetIdleVFXComponent))
  ;[1 0.2176 0.2574 1 200] EVENT 🖥️
  (|CallParentFunction)
  (Utilities|FlowControl|Sequence
    (:then_0
      ;[0.141 0.4403 1 1 200] FRONTEND 🛒
      (if (or (Reactive|Frontend|IsCosmeticObjectDisplayedasPreview 0) (CustomCosmeticModifier|IsinFrontEnd))
        (Effects|Components|ParticleSystem|SetBoolParameter (CallFunction|GetIdleVFXComponent) "bIsFrontEnd" true)
        (Effects|Components|ParticleSystem|SetBoolParameter (CallFunction|GetIdleVFXComponent) "bIsFrontEndPreview" true)
        (Utilities|FlowControl|RetriggerableDelay 3.0)
        (bind _returnvalue_3 (PartModifier|SpawnSystem|SpawnSystemAttachedPartModifier "/BRCosmetics/Accessories/FORT_Backpacks/Backpack_F_MED_RebelClaw/FX/NS_Backpack_RebelClaw_Smoke.NS_Backpack_RebelClaw_Smoke" (CustomCharacter|GetSkeletalMeshForPartOnPlayerPawn) "Root")
          (:Success
            (if (or (Reactive|Frontend|IsCosmeticObjectDisplayedasPreview 0) (CustomCosmeticModifier|IsinFrontEnd))
              (Effects|Components|ParticleSystem|SetBoolParameter _returnvalue_3 "bIsFrontEnd" true)
              (Effects|Components|ParticleSystem|SetBoolParameter _returnvalue_3 "bIsFrontEndPreview" true)
              (bind _index (Class|FortCosmeticFunctionLibrary|GetVariantChannelActiveIndexForSlot "(TagName=\"LoadoutSlot.Backpack\")" (CosmeticLoadout|GetCosmeticLoadout) "(TagName=\"Cosmetics.Variant.Channel.Progressive\")" (Utilities|Array|MakeArray "(TagName=\"Cosmetics.Variant.Property.Stage1\")" "(TagName=\"Cosmetics.Variant.Property.Stage2\")")))
              (if (== _index 0)
                (Effects|Components|ParticleSystem|SetColorParameter _returnvalue_3 "StartColor" "(R=0.109376,G=0.266993,B=0.300000,A=1.000000)")
                (Effects|Components|ParticleSystem|SetColorParameter _returnvalue_3 "EndColor" "(R=0.250000,G=0.250000,B=0.250000,A=1.000000)"))
              (Effects|Components|ParticleSystem|SetColorParameter _returnvalue_3 "StartColor" "(R=0.109376,G=0.266993,B=0.300000,A=1.000000)")
              (Effects|Components|ParticleSystem|SetColorParameter _returnvalue_3 "EndColor" "(R=0.140454,G=0.102357,B=0.300000,A=1.000000)"))
            (Effects|Components|ParticleSystem|SetBoolParameter _returnvalue_3 "bIsFrontEnd" false)
            (Effects|Components|ParticleSystem|SetBoolParameter _returnvalue_3 "bIsFrontEndPreview" false)))
        (else
          (Effects|Components|ParticleSystem|SetBoolParameter (CallFunction|GetIdleVFXComponent) "bIsFrontEnd" false)
          (Effects|Components|ParticleSystem|SetBoolParameter (CallFunction|GetIdleVFXComponent) "bIsFrontEndPreview" false))))
    (:then_1
      ;[1 0.7262 0.2176 1 200] ON KILL REACTIVITY ☠
      (bind (_statname _statvalue) (Cosmetics|WatchCosmeticStat (CustomCharacter|GetAssociatedPlayerPawn) "(TagName=\"StatManager.AthenaKills\")" 0 false)
        (:OnStatChanged
          (CallFunction|DEBUGPrintString "The kill worked")
          (if (> _statvalue 0)
            (Utilities|FlowControl|Delay 0.5)
            (bind _returnvalue_4 (PartModifier|SpawnSystem|SpawnSystemAttachedPartModifier "/BRCosmetics/Accessories/FORT_Backpacks/Backpack_F_MED_RebelClaw/FX/NS_Backpack_RebelClaw_Smoke.NS_Backpack_RebelClaw_Smoke" (CustomCharacter|GetSkeletalMeshForPartOnPlayerPawn) "Root")
              (:Success
                (bind _index_2 (Class|FortCosmeticFunctionLibrary|GetVariantChannelActiveIndexForSlot "(TagName=\"LoadoutSlot.Backpack\")" (CosmeticLoadout|GetCosmeticLoadout) "(TagName=\"Cosmetics.Variant.Channel.Progressive\")" (Utilities|Array|MakeArray "(TagName=\"Cosmetics.Variant.Property.Stage1\")" "(TagName=\"Cosmetics.Variant.Property.Stage2\")")))
                (if (== _index_2 0)
                  (Effects|Components|ParticleSystem|SetColorParameter _returnvalue_4 "StartColor" "(R=0.032812,G=0.066171,B=0.090000,A=1.000000)")
                  (Effects|Components|ParticleSystem|SetColorParameter _returnvalue_4 "EndColor" "(R=0.070000,G=0.070000,B=0.070000,A=1.000000)")
                  (else
                    (Effects|Components|ParticleSystem|SetColorParameter _returnvalue_4 "StartColor" "(R=0.025521,G=0.051467,B=0.070000,A=1.000000)")
                    (Effects|Components|ParticleSystem|SetColorParameter _returnvalue_4 "EndColor" "(R=0.028091,G=0.020471,B=0.060000,A=1.000000)")))
                (Effects|Components|ParticleSystem|SetFloatParameter (CallFunction|GetIdleVFXComponent) "SpawnScale" 0.0)
                (Utilities|FlowControl|Delay 2.5)
                (Effects|Components|ParticleSystem|SetFloatParameter (CallFunction|GetIdleVFXComponent) "SpawnScale" 1.0)))))))
    (:then_2
      (bind _index_3 (Class|FortCosmeticFunctionLibrary|GetVariantChannelActiveIndexForSlot "(TagName=\"LoadoutSlot.Backpack\")" (CosmeticLoadout|GetCosmeticLoadout) "(TagName=\"Cosmetics.Variant.Channel.Progressive\")" (Utilities|Array|MakeArray "(TagName=\"Cosmetics.Variant.Property.Stage1\")" "(TagName=\"Cosmetics.Variant.Property.Stage2\")")))
      (if (== _index_3 0)
        (Effects|Components|ParticleSystem|SetColorParameter _returnvalue "StartColor" "(R=0.091144,G=0.183810,B=0.250000,A=1.000000)")
        (Effects|Components|ParticleSystem|SetColorParameter _returnvalue "EndColor" "(R=0.120000,G=0.120000,B=0.120000,A=1.000000)")
        (else
          (Effects|Components|ParticleSystem|SetColorParameter _returnvalue "StartColor" "(R=0.072916,G=0.147049,B=0.200000,A=1.000000)")
          (Effects|Components|ParticleSystem|SetColorParameter _returnvalue "EndColor" "(R=0.056181,G=0.040943,B=0.120000,A=1.000000)"))))
    (:then_3
      (Utilities|IsSkydiving (CustomCharacter|GetAssociatedPlayerPawn)
        (:true
          (Effects|Components|ParticleSystem|SetFloatParameter _returnvalue_2 "Interpolate A" 3.0)
          (Effects|Components|ParticleSystem|SetFloatParameter _returnvalue_2 "Interpolate B" 6.0))
        (:false
          (Effects|Components|ParticleSystem|SetFloatParameter _returnvalue_2 "Interpolate A" 2e+01)
          (Effects|Components|ParticleSystem|SetFloatParameter _returnvalue_2 "Interpolate B" 5e+01))))))

(fn ConstructionScript
  (|CallParentFunction))
