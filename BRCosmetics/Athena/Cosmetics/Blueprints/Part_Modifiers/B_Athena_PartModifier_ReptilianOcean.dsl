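; Cosmetic part modifier for the ReptilianOcean character that makes eyes, face tattoo and idle VFX react to aiming down sights.
; Targeting changes, gated by the ReactiveOn meta tag, trigger ReactivityEvent locally and via server-to-client events.
; ReactivityEvent sets the VFX "ADS" parameter and plays or reverses EyesTimeline to drive material parameters.
; Initialization reads the Progressive variant to set emissive levels; in frontend display, a looping timeline toggles the reaction.
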(event EventOnPostResIn
  ;[0 0 0] post res in
  (|CallParentFunction)
  (Utilities|FlowControl|Sequence
    (:then_0
      (CallFunction|InitializePlayerReactivity))
    (:then_1
      (if (Reactive|Frontend|IsCosmeticObjectDisplayedinFrontendContext 0)
        (Timeline|FrontendTimeline Play
          (:Finished
            (bind _output_get (Variables|Default|SetFrontendToggleReactivity (not (Variables|Default|GetFrontendToggleReactivity))))
            (CallFunction|ReactivityEvent _output_get)
            (Timeline|FrontendTimeline PlayFromStart)))))))

(event EventOnTargetingOrFirstPersonCameraChanged
  (bind _returnvalue (CustomCharacter|GetAssociatedPlayerPawn))
  (bind _bistargeting (Variables|/Script/FortniteGame.FortPlayerPawn|GetbIsTargeting _returnvalue))
  ;[0 0 0] targeting changed
  (|CallParentFunction)
  (if (GameplayTags|HasTag (Cosmetics|GatherMetaTags (CustomCharacter|GetAssociatedPlayerPawn) "/BRCosmetics/Athena/Items/Cosmetics/Characters/Character_ReptilianOcean_Sleek.Character_ReptilianOcean_Sleek" :!class /Script/FortniteGame.FortPlayerPawn) "(TagName=\"Cosmetics.MetaTags.ReactiveOn\")")
    (CallFunction|ReactivityEvent _bistargeting)
    (CallFunction|CEFXServer _bistargeting) ; It calls the Server Event, that calls the Client Event, \nthat finally calls the reactivity event
  ))

(event Custom|ReactivityEvent
  (params (InADSing :type Boolean))
  (bind _returnvalue (CustomCharacter|GetSkeletalMeshForPartOnPlayerPawn))
  (bind _material_instance_dynamic (Class|BTechArtGlobalLibrary|GetDynamicMaterialInstance _returnvalue 1))
  (bind _material_instance_dynamic_2 (Class|BTechArtGlobalLibrary|GetDynamicMaterialInstance _returnvalue))
  ;[0 0 0] reactivity event
  (if (!= InADSing (Variables|Default|GetcachedAdsing))
    (Utilities|FlowControl|Sequence
      (:then_0
        (Variables|Default|SetcachedAdsing (Math|Conversions|ToInteger(Boolean) InADSing))
        (Effects|Components|ParticleSystem|SetBoolParameter (CallFunction|GetIdleVFXComponent) "ADS" InADSing))
      (:then_1
        (if InADSing
          (bind (_direction _eyes_0_1) (Timeline|EyesTimeline PlayFromStart)
            (:Update
              (Rendering|Material|SetScalarParameterValue _material_instance_dynamic_2 "EyeEffect" _eyes_0_1 :!class /Script/Engine.MaterialInstanceDynamic) ; Set face tattoo
              (Rendering|Material|SetScalarParameterValue _material_instance_dynamic "emissive mult" _eyes_0_1 :!class /Script/Engine.MaterialInstanceDynamic) ; Set Eyes to Glow
              (Rendering|Material|SetScalarParameterValue _material_instance_dynamic "Eye Diffuse Multiplier" (Math|Conversions|ToFloat(Boolean) (not InADSing)) :!class /Script/Engine.MaterialInstanceDynamic) ; Disable Eyes Diffuse
              (Rendering|Material|SetScalarParameterValue _material_instance_dynamic_2 "SnakesEyesEmissiveIntensity" (+ (* _eyes_0_1 (Variables|Default|GetMaxEmissiveValue)) (Variables|Default|GetbaseEmissiveValue)) :!class /Script/Engine.MaterialInstanceDynamic) ; set snake eye colors
            ))
          (else
            (Timeline|EyesTimeline ReverseFromEnd)))))))

(event Custom|CE_FX_Server
  (params (InADSing :type Boolean))
  ;[1 0.2356 0.1858] "Server Logic". It makes sure the reactivity can be seen by other clients
  (CallFunction|CEFXClient InADSing))

(event Custom|CE_FX_Client
  (params (InADSing :type Boolean))
  ;[1 0.2356 0.1858] "Server Logic". It makes sure the reactivity can be seen by other clients
  (CallFunction|ReactivityEvent InADSing))

(fn InitializePlayerReactivity
  (bind _asfort_player_pawn (Utilities|Casting|CastToFortPlayerPawn (Actor|GetOwner)))
  (bind _index (CosmeticLoadout|GetVariantChannelActiveIndexforSlotActor "(TagName=\"LoadoutSlot.Character\")" _asfort_player_pawn "(TagName=\"Cosmetics.Variant.Channel.Progressive\")" (Utilities|Array|MakeArray "(TagName=\"Cosmetics.Variant.Property.Stage1\")" "(TagName=\"Cosmetics.Variant.Property.Stage2\")" "(TagName=\"Cosmetics.Variant.Property.Stage3\")")))
  (bind _output_get (Variables|Default|SetVariantIndex _index))
  (switch int _output_get
    (:Default
      _)
    (:0
      _)
    (:1
      (Variables|Default|SetbaseEmissiveValue 1.0)
      (Variables|Default|SetMaxEmissiveValue 4.0)
      (Rendering|Material|SetScalarParameterValue (Class|BTechArtGlobalLibrary|GetDynamicMaterialInstance (CustomCharacter|GetSkeletalMeshForPartOnPlayerPawn)) "SnakesEyesEmissiveIntensity" (Variables|Default|GetbaseEmissiveValue) :!class /Script/Engine.MaterialInstanceDynamic) ; set snake eye colors
    )
    (:2
      _))
  (Variables|Default|SetbaseEmissiveValue 1.0)
  (Variables|Default|SetMaxEmissiveValue 2.5))

(fn SetSnakesEmissiveIntensity
  (params (Index :type Boolean))
  (bind _returnvalue (CustomCharacter|GetSkeletalMeshForPartOnPlayerPawn))
  (bind _material_instance_dynamic (Class|BTechArtGlobalLibrary|GetDynamicMaterialInstance _returnvalue))
  (bind _maxemissivevalue (Variables|Default|GetMaxEmissiveValue))
  (bind _baseemissivevalue (Variables|Default|GetbaseEmissiveValue))
  (bind _returnvalue_2 (select Index _maxemissivevalue _baseemissivevalue))
  (switch int (Variables|Default|GetVariantIndex)
    (:Default
      _)
    (:0
      _)
    (:1
      (Rendering|Material|SetScalarParameterValue _material_instance_dynamic "SnakesEyesEmissiveIntensity" _returnvalue_2 :!class /Script/Engine.MaterialInstanceDynamic) ; set snake eye colors
    ))
  (Rendering|Material|SetScalarParameterValue _material_instance_dynamic "EmissiveBrightness" _returnvalue_2 :!class /Script/Engine.MaterialInstanceDynamic) ; set snake eye colors
)

(fn ConstructionScript
  (|CallParentFunction))
