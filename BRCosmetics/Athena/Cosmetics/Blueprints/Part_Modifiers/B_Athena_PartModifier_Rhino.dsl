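; Cosmetic part modifier for the Rhino character part that switches pipe smoke effects with the pawn's movement.
; Skydiving and jumping deactivate the idle VFX; ending skydiving or landing reactivates it.
; Jumping also reactivates a cached jump smoke emitter, or spawns a debounced one snapped to the dyn_Pipe socket on the part's skeletal mesh when none is valid.
; Landing deactivates that jump emitter.
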(event EventOnBeginSkydiving
  (|CallParentFunction)
  (Components|Activation|Deactivate (CallFunction|GetIdleVFXComponent)))

(event EventOnEndSkydiving
  (|CallParentFunction)
  (Components|Activation|Activate (CallFunction|GetIdleVFXComponent) false))

(event EventOnPawnJumped
  (bind _spawned_jump_fx (Variables|Default|GetSpawned%20Jump%20FX))
  (|CallParentFunction)
  (Components|Activation|Deactivate (CallFunction|GetIdleVFXComponent))
  (Utilities|IsValid _spawned_jump_fx
    (:Is%20Valid
      (Components|Activation|Activate _spawned_jump_fx true))
    (:Is%20Not%20Valid
      (bind _returnvalue (PartModifier|SpawnSystem|SpawnEmitterAttachedPartModifier(Legacy) :EmitterTemplate (Variables|Default|GetJumping%20Local%20Smoke) :AttachToComponent (CustomCharacter|GetSkeletalMeshForPartOnPlayerPawn) :AttachPointName "dyn_Pipe" :Scale "1.000000,1.000000,1.000000" :LocationType "SnapToTargetIncludingScale" :bAutoDestroy true :PoolingMethod "None" :bAutoActivate true :bMatchPawnCustomDepth true :bDebounce true :DebounceTime -1.0)
        (:Success
          (Variables|Default|SetSpawned%20Jump%20FX _returnvalue))))))

(event EventOnPawnLanded
  (bind _spawned_jump_fx (Variables|Default|GetSpawned%20Jump%20FX))
  (|CallParentFunction)
  (Components|Activation|Activate (CallFunction|GetIdleVFXComponent) false)
  (Utilities|IsValid _spawned_jump_fx
    (:Is%20Valid
      (Components|Activation|Deactivate _spawned_jump_fx))))

(fn ConstructionScript)
