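; Cosmetic part modifier for the RhombCamo character part, set up for a reactive camo look.
; On res-in, it plays an FE_Preview timeline in the front end, otherwise it watches the Athena kills stat.
; Both the timeline and stat callbacks are empty here.
; It also writes current game time into the RC_GameTime_CCPM material parameter on the part's skeletal mesh and the pawn's Face mesh.
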(event EventOnPostResIn
  (bind _returnvalue (CustomCharacter|GetAssociatedPlayerPawn))
  (bind _returnvalue_2 (Utilities|Time|GetGameTimeinSeconds))
  (|CallParentFunction)
  (if (CustomCosmeticModifier|IsinFrontEnd)
    (Timeline|FE_Preview PlayFromStart
      (:Reactivity
        _))
    (else
      (Cosmetics|WatchCosmeticStat _returnvalue "(TagName=\"StatManager.AthenaKills\")" 0 false
        (:OnStatChanged
          _))))
  (Rendering|Material|SetScalarParameterValueonMaterials (CustomCharacter|GetSkeletalMeshForPartOnPlayerPawn) "RC_GameTime_CCPM" _returnvalue_2)
  (Rendering|Material|SetScalarParameterValueonMaterials (Customization|GetSkeletalMeshforPartType _returnvalue "Face" :!class /Script/FortniteGame.FortPlayerPawn) "RC_GameTime_CCPM" _returnvalue_2))

(event EventOnExitVehicle
  (|CallParentFunction) ; This node is disabled and will not be called.\nDrag off pins to build functionality.
)

(event EventOnTargetingOrFirstPersonCameraChanged)

(event EventOnHitPawn
  (params (SelfActor :type Actor%20Object%20Reference) (OtherActor :type Actor%20Object%20Reference) (NormalImpulse :type Vector :default "0, 0, 0") (Hit :type Hit%20Result%20Structure%20%28by%20ref%29)))

(fn ConstructionScript
  (|CallParentFunction))
