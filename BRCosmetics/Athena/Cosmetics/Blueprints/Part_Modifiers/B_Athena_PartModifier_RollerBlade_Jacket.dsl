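; Cosmetic part modifier for the Athena RollerBlade jacket that adapts the character's idle VFX to the selected variants.
; After the parent post-resolve event, it reads the pawn's Material variant channel to decide whether the idle VFX component is activated or deactivated.
; When a non-default material is active, the Parts channel index sets the particle boolean JacketVariantActive, true only for the first stage.
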(event EventOnPostResIn
  (bind _returnvalue (CallFunction|GetIdleVFXComponent))
  (bind _returnvalue_2 (CustomCharacter|GetAssociatedPlayerPawn))
  (|CallParentFunction)
  (bind (_index _bindexisvalid) (CosmeticLoadout|GetVariantChannelActiveIndexforSlotActor "(TagName=\"LoadoutSlot.Character\")" _returnvalue_2 "(TagName=\"Cosmetics.Variant.Channel.Material\")" (Utilities|Array|MakeArray "(TagName=\"Cosmetics.Variant.Property.Mat1\")" "(TagName=\"Cosmetics.Variant.Property.Mat2\")")))
  (if _bindexisvalid
    (if (> _index 0)
      (Components|Activation|Activate _returnvalue false)
      (bind _index_2 (CosmeticLoadout|GetVariantChannelActiveIndexforSlotActor "(TagName=\"LoadoutSlot.Character\")" _returnvalue_2 "(TagName=\"Cosmetics.Variant.Channel.Parts\")" (Utilities|Array|MakeArray "(TagName=\"Cosmetics.Variant.Property.Stage1\")" "(TagName=\"Cosmetics.Variant.Property.Stage2\")")))
      (if (> _index_2 0)
        (Effects|Components|ParticleSystem|SetBoolParameter _returnvalue "JacketVariantActive" false)
        (else
          (Effects|Components|ParticleSystem|SetBoolParameter _returnvalue "JacketVariantActive" true)))
      (else
        _)))
  (Components|Activation|Deactivate _returnvalue))

(fn ConstructionScript
  (|CallParentFunction))
