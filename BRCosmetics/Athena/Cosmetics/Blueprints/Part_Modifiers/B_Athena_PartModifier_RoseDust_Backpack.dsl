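; Reactive part modifier for the RoseDust backpack that swaps its idle Niagara effect between kill, damage and weak-point looks.
; In matches, cosmetic stat watchers on the owning pawn trigger each look; in the frontend, a looping timer cycles through all three.
; Each look sets material and particle parameters, restarts the system, then reverts to the default "Meanwhile" look after a short delay.
; An IsKilled flag stops damage reactions from overriding the kill look.
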(event Custom|FrontEnd%20Cycle
  (switch int (Variables|Default|GetFrontEndInt)
    (:0
      (CallFunction|Kills))
    (:1
      (CallFunction|Damage))
    (:2
      (CallFunction|WeakPoint)
      (Variables|Default|SetFrontEndInt 0)))
  (Variables|Default|SetFrontEndInt (+ (Variables|Default|GetFrontEndInt) 1)))

(event Custom|Kills
  (bind _returnvalue (CallFunction|GetIdleVFXComponent))
  (bind _asniagara_particle_system_component (Utilities|Casting|CastToNiagaraComponent _returnvalue))
  (Niagara|SetNiagaraVariable(Material) _asniagara_particle_system_component "Material" "/BRCosmetics/Accessories/FORT_Backpacks/Backpack_F_MED_Rosedust/Materials/MI_Rosedust_Wabooff_EN_MI.MI_Rosedust_Wabooff_EN_MI")
  (Effects|Components|ParticleSystem|SetColorParameter _asniagara_particle_system_component "IconColor" "(R=1.000000,G=1.000000,B=0.000000,A=1.000000)")
  (Effects|Components|ParticleSystem|SetBoolParameter _asniagara_particle_system_component "NonUniformSizeSwitch" true)
  (Effects|Components|ParticleSystem|SetIntParameter _asniagara_particle_system_component "StateColorInt" 3)
  (Effects|Components|ParticleSystem|SetIntParameter _asniagara_particle_system_component "PivotOffsetArrayNr" 3)
  (Effects|Components|ParticleSystem|SetIntParameter _asniagara_particle_system_component "AdditionalElementInt" 3)
  (Effects|Components|ParticleSystem|SetIntParameter _asniagara_particle_system_component "StreakVisInt" 1)
  (Effects|Components|ParticleSystem|SetVectorParameter _asniagara_particle_system_component "EffectChannelPicker" "1.000000, 0,1.000000")
  (Niagara|ResetSystem _asniagara_particle_system_component)
  (Utilities|FlowControl|Delay 2.0)
  (CallFunction|ResetToMeanwhile)
  (Variables|Default|SetIsKilled false))

(event EventOnPostResIn
  (|CallParentFunction)
  (if (Reactive|Frontend|IsCosmeticObjectDisplayedinFrontendContext 0)
    (Utilities|Time|SetTimerbyEvent (EventDispatchers|CreateEvent :!event Custom|FrontEnd%20Cycle) 6.0 true)
    (else
      (Utilities|FlowControl|Sequence
        (:then_0
          (bind (_statname _statvalue) (Cosmetics|WatchCosmeticStat (CustomCharacter|GetAssociatedPlayerPawn) "(TagName=\"StatManager.AthenaKills\")" 0 false)
            (:OnStatChanged
              (if (> _statvalue 0)
                (Variables|Default|SetIsKilled true)
                (CallFunction|Kills)))))
        (:then_1
          (bind (_statname_2 _statvalue_2) (Cosmetics|WatchCosmeticStat (CustomCharacter|GetAssociatedPlayerPawn) "(TagName=\"StatManager.DamageDealt\")" 0 false)
            (:OnStatChanged
              (if (> _statvalue_2 0)
                (CallFunction|Damage)))))
        (:then_2
          (bind (_statname_3 _statvalue_3) (Cosmetics|WatchCosmeticStat (CustomCharacter|GetAssociatedPlayerPawn) "(TagName=\"StatManager.WeakPointHit\")" 0 false)
            (:OnStatChanged
              (if (> _statvalue_3 0)
                (CallFunction|WeakPoint)))))))))

(event Custom|Damage
  (bind _returnvalue (CallFunction|GetIdleVFXComponent))
  (bind _asniagara_particle_system_component (Utilities|Casting|CastToNiagaraComponent _returnvalue))
  (if (not (Variables|Default|GetIsKilled))
    (Niagara|SetNiagaraVariable(Material) _asniagara_particle_system_component "Material" "/BRCosmetics/Accessories/FORT_Backpacks/Backpack_F_MED_Rosedust/Materials/MI_Rosedust_Kawap_EN_MI.MI_Rosedust_Kawap_EN_MI")
    (Effects|Components|ParticleSystem|SetColorParameter _asniagara_particle_system_component "IconColor" "(R=0.098371,G=0.253362,B=1.000000,A=1.000000)")
    (Effects|Components|ParticleSystem|SetBoolParameter _asniagara_particle_system_component "NonUniformSizeSwitch" true)
    (Effects|Components|ParticleSystem|SetIntParameter _asniagara_particle_system_component "StateColorInt" 1)
    (Effects|Components|ParticleSystem|SetIntParameter _asniagara_particle_system_component "AdditionalElementInt" 1)
    (Effects|Components|ParticleSystem|SetIntParameter _asniagara_particle_system_component "StreakVisInt" 0)
    (Effects|Components|ParticleSystem|SetIntParameter _asniagara_particle_system_component "PivotOffsetArrayNr" 1)
    (Effects|Components|ParticleSystem|SetVectorParameter _asniagara_particle_system_component "EffectChannelPicker" "0.000000, 0,1.000000")
    (Niagara|ResetSystem _asniagara_particle_system_component)
    (Utilities|FlowControl|Delay 2.0)
    (CallFunction|ResetToMeanwhile)
    (Variables|Default|SetIsKilled false)))

(event Custom|Weak%20Point
  (bind _returnvalue (CallFunction|GetIdleVFXComponent))
  (bind _asniagara_particle_system_component (Utilities|Casting|CastToNiagaraComponent _returnvalue))
  (Niagara|SetNiagaraVariable(Material) _asniagara_particle_system_component "Material" "/BRCosmetics/Accessories/FORT_Backpacks/Backpack_F_MED_Rosedust/Materials/MI_Rosedust_Thauk_EN_MI.MI_Rosedust_Thauk_EN_MI")
  (Effects|Components|ParticleSystem|SetColorParameter _asniagara_particle_system_component "IconColor" "(R=0.251736,G=0.144760,B=0.046364,A=1.000000)")
  (Effects|Components|ParticleSystem|SetBoolParameter _asniagara_particle_system_component "NonUniformSizeSwitch" true)
  (Effects|Components|ParticleSystem|SetIntParameter _asniagara_particle_system_component "StateColorInt" 2)
  (Effects|Components|ParticleSystem|SetIntParameter _asniagara_particle_system_component "AdditionalElementInt" 2)
  (Effects|Components|ParticleSystem|SetIntParameter _asniagara_particle_system_component "PivotOffsetArrayNr" 2)
  (Effects|Components|ParticleSystem|SetVectorParameter _asniagara_particle_system_component "EffectChannelPicker" "0.000000,1.000000, 0")
  (Niagara|ResetSystem _asniagara_particle_system_component)
  (Utilities|FlowControl|Delay 2.0)
  (CallFunction|ResetToMeanwhile)
  (Variables|Default|SetIsKilled false))

(event Custom|Reset%20to%20Meanwhile
  (bind _returnvalue (CallFunction|GetIdleVFXComponent))
  (bind _asniagara_particle_system_component (Utilities|Casting|CastToNiagaraComponent _returnvalue))
  (Niagara|SetNiagaraVariable(Material) _asniagara_particle_system_component "Material" "/BRCosmetics/Accessories/FORT_Backpacks/Backpack_F_MED_Rosedust/Materials/MI_Rosedust_Meanwhile_EN_MI.MI_Rosedust_Meanwhile_EN_MI")
  (Effects|Components|ParticleSystem|SetColorParameter _asniagara_particle_system_component "IconColor" "(R=1.000000,G=0.000000,B=0.000000,A=1.000000)")
  (Effects|Components|ParticleSystem|SetBoolParameter _asniagara_particle_system_component "NonUniformSizeSwitch" false)
  (Effects|Components|ParticleSystem|SetIntParameter _asniagara_particle_system_component "StateColorInt" 0)
  (Effects|Components|ParticleSystem|SetIntParameter _asniagara_particle_system_component "AdditionalElementInt" 0)
  (Effects|Components|ParticleSystem|SetIntParameter _asniagara_particle_system_component "PivotOffsetArrayNr" 0)
  (Effects|Components|ParticleSystem|SetVectorParameter _asniagara_particle_system_component "EffectChannelPicker" "1.000000, 0, 0")
  (Niagara|ResetSystem _asniagara_particle_system_component))

(fn ConstructionScript
  (|CallParentFunction))
