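; Character part modifier for the RoseScour tail cosmetic that adjusts its idle Niagara effect to suit the equipped backpack.
; After the parent's post-res-in logic runs, it reads the associated pawn's cosmetic loadout and casts the idle VFX component to Niagara.
; It sets bHasCape true when no backpack is equipped or the backpack has the Cosmetics.Backpack.IsCape tag.
; The other event handlers are unused stubs.
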(event EventOnPostResIn
  (|CallParentFunction) ; This node is disabled and will not be called.\nDrag off pins to build functionality.
)

(event EventOnDayPhaseChanged
  (params (CurrentDayPhase :type EFortDayPhase%20Enum :default "Morning") (PreviousDayPhase :type EFortDayPhase%20Enum :default "Morning") (bAtCreation :type Boolean)))

(event EventOnNewPlayerWeaponType
  (params (NewType :type EFortItemType%20Enum :default "WorldItem") (OldType :type EFortItemType%20Enum :default "WorldItem")))

(event EventOnWeaponChanged
  (params (NewWeapon :type Fort%20Weapon%20Object%20Reference) (OldWeapon :type Fort%20Weapon%20Object%20Reference))
  (|CallParentFunction NewWeapon OldWeapon) ; This node is disabled and will not be called.\nDrag off pins to build functionality.
)

(event EventOnWeaponFired)

(event EventOnNewWeapons
  (params (WeaponsAsItems :type Array%20of%20Fort%20Item%20Object%20References)))

(event EventOnCurieStateAttached
  (params (Tag :type Gameplay%20Tag%20Structure%20%28by%20ref%29)))

(event EventOnCurieStateDetached
  (params (Tag :type Gameplay%20Tag%20Structure%20%28by%20ref%29)))

(event EventOnBeginSkydiving
  (|CallParentFunction) ; This node is disabled and will not be called.\nDrag off pins to build functionality.
)

(event EventOnEndSkydiving
  (|CallParentFunction) ; This node is disabled and will not be called.\nDrag off pins to build functionality.
)

(event EventOnEnterVehicle
  (params (VehicleEntered :type Fort%20Vehicle%20Interface%20Interface%20%28by%20ref%29))
  (|CallParentFunction VehicleEntered) ; This node is disabled and will not be called.\nDrag off pins to build functionality.
)

(event EventOnExitVehicle
  (|CallParentFunction) ; This node is disabled and will not be called.\nDrag off pins to build functionality.
)

(event EventOnTargetingOrFirstPersonCameraChanged)

(event EventOnHitPawn
  (params (SelfActor :type Actor%20Object%20Reference) (OtherActor :type Actor%20Object%20Reference) (NormalImpulse :type Vector :default "0, 0, 0") (Hit :type Hit%20Result%20Structure%20%28by%20ref%29)))

(event EventOnPostResIn
  (bind (_bannericonid _bannercolorid _skydivecontrail _glider _pickaxe _character _charactervariantchannels _backpack) (Utilities|Struct|BreakFortAthenaLoadout (Cosmetics|GetPawnCosmeticLoadout (CustomCharacter|GetAssociatedPlayerPawn) false false)))
  (bind _returnvalue (CallFunction|GetIdleVFXComponent))
  (|CallParentFunction)
  (Utilities|IsValid _returnvalue
    (:Is%20Valid
      (bind _asniagara_particle_system_component (Utilities|Casting|CastToNiagaraComponent _returnvalue))
      (Niagara|SetNiagaraVariable(Bool) _asniagara_particle_system_component "bHasCape" (or (not (Utilities|IsValid _backpack :!class /Script/Engine.KismetSystemLibrary)) (GameplayTags|HasMatchingGameplayTag _backpack "(TagName=\"Cosmetics.Backpack.IsCape\")"))))))

(fn ConstructionScript
  (|CallParentFunction))
