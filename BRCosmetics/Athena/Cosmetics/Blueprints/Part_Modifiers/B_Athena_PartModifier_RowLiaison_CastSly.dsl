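; Cosmetic part modifier for the RowLiaison CastSly character that plays an elimination-reactive effect on its fourth Progressive variant stage.
; After post-resurrection init, it reads the active stage from the character loadout and acts only at index 3.
; The effect stamps game time into the dynamic material's RLCS_GameTimeFromBP parameter and spawns an attached elimination Niagara system.
; It fires immediately as a front-end preview, or in match whenever the watched AthenaKills stat changes.
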(event EventOnPostResIn
  (bind _returnvalue (CustomCharacter|GetSkeletalMeshForPartOnPlayerPawn))
  (bind _material_instance_dynamic (Class|BTechArtGlobalLibrary|GetDynamicMaterialInstance _returnvalue))
  (bind _returnvalue_2 (Utilities|Time|GetGameTimeinSeconds))
  (bind _returnvalue_3 (CustomCharacter|GetAssociatedPlayerPawn))
  (|CallParentFunction)
  (bind _asplayer_pawn_athena (Utilities|Casting|CastToPlayerPawn_Athena _returnvalue_3))
  (bind _index (CosmeticLoadout|GetVariantChannelActiveIndexforSlotActor "(TagName=\"LoadoutSlot.Character\")" _asplayer_pawn_athena "(TagName=\"Cosmetics.Variant.Channel.Progressive\")" (Utilities|Array|MakeArray "(TagName=\"Cosmetics.Variant.Property.Stage1\")" "(TagName=\"Cosmetics.Variant.Property.Stage2\")" "(TagName=\"Cosmetics.Variant.Property.Stage3\")" "(TagName=\"Cosmetics.Variant.Property.Stage4\")" "(TagName=\"Cosmetics.Variant.Property.Stage5\")") false false))
  (switch int _index
    (:3
      (if (CustomCosmeticModifier|IsinFrontEnd)
        ;[1 1 1] FE preview cycle
        (Rendering|Material|SetScalarParameterValue _material_instance_dynamic "RLCS_GameTimeFromBP" _returnvalue_2 :!class /Script/Engine.MaterialInstanceDynamic)
        ;;[1 1 1] Fx
        (PartModifier|SpawnSystem|SpawnSystemAttachedPartModifier "/BRCosmetics/Characters/Player/Female/Medium/Bodies/F_MED_RowLiaison_Cast/FX/NS_Character_RowLiaison_CastSly_elim.NS_Character_RowLiaison_CastSly_elim" (CustomCharacter|GetSkeletalMeshForPartOnPlayerPawn)
          (:Success
            (Utilities|FlowControl|Delay 8.0)))
        (else
          (Cosmetics|WatchCosmeticStat _returnvalue_3 "(TagName=\"StatManager.AthenaKills\")" 0 false
            (:OnStatChanged
              ;[1 1 1] In game
              (Rendering|Material|SetScalarParameterValue _material_instance_dynamic "RLCS_GameTimeFromBP" _returnvalue_2 :!class /Script/Engine.MaterialInstanceDynamic)
              ;[1 1 1] FX
              (PartModifier|SpawnSystem|SpawnSystemAttachedPartModifier "/BRCosmetics/Characters/Player/Female/Medium/Bodies/F_MED_RowLiaison_Cast/FX/NS_Character_RowLiaison_CastSly_elim.NS_Character_RowLiaison_CastSly_elim" (CustomCharacter|GetSkeletalMeshForPartOnPlayerPawn)))))))))

(fn ConstructionScript
  (|CallParentFunction))
