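; Character part modifier Blueprint for the Ruckus cosmetic that adjusts its painted toon line thickness.
; Cosmetic float 0 supplies the multiplier and registers a WhenCosmeticPartApplied callback on the associated player pawn, storing the delegate handle.
; The callback writes the multiplier to the DF_LineThicknessMult material parameter on the applied part's mesh.
; On post-res-in, it runs the parent logic, then a toon frontend LUT replacement.
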(event EventOnPostResIn
  (|CallParentFunction)
  (Utilities|ToonFrontendLUTReplace))

(event EventsetCosmeticFloat0
  (params (float :type Float%20%28double-precision%29))
  (Fortnite|Utility|FortBlueprintLog "Ruckus CCPM - Called Cosmetic Float 0")
  (CallFunction|SetPaintedLineThicknessMult float))

(event Custom|Set%20Painted%20Line%20Thickness%20Mult
  (params (LineThicknessMult :type Float%20%28double-precision%29))
  (bind _delegatehandlecontroller_out (Cosmetics|WhenCosmeticPartApplied (CustomCharacter|GetAssociatedPlayerPawn) (delegate Custom|OnCosmeticPart) true))
  (Variables|Default|SetDelegate%20Handle%20Controller _delegatehandlecontroller_out))

(event Custom|OnCosmeticPart
  (params (Pawn :type Fort%20Player%20Pawn%20Object%20Reference) (Part :type Custom%20Character%20Part%20Object%20Reference) (MeshPart :type Skeletal%20Mesh%20Component%20Object%20Reference) (PartType :type EFortCustomPartType%20Enum :default "Head"))
  (Rendering|Material|SetScalarParameterValueonMaterials MeshPart "DF_LineThicknessMult" (stale Custom|Set%20Painted%20Line%20Thickness%20Mult LineThicknessMult)))

(fn ConstructionScript
  (|CallParentFunction))
