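; Cosmetic part modifier Blueprint for the Ruckus Mini character that adjusts material color grading on its custom character part outside the front end.
; On post-res-in, it calls the parent, then checks whether it is in the front end.
; In the front end, a switch on part type has empty Head, Body and Face branches.
; Otherwise, it sets the FTColorGrading scalar on the associated pawn's skeletal mesh for that part.
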(event EventOnPostResIn
  (bind _returnvalue (CustomCharacter|GetCharacterPartType))
  (|CallParentFunction)
  (if (CustomCosmeticModifier|IsinFrontEnd)
    (switch Utilities|FlowControl|Switch|SwitchonEFortCustomPartType _returnvalue
      (:Head
        _)
      (:Body
        _)
      (:Face
        _))
    (Rendering|Material|SetScalarParameterValueonMaterials (Customization|GetSkeletalMeshforPartType (CustomCharacter|GetAssociatedPlayerPawn) _returnvalue :!class /Script/FortniteGame.FortPlayerPawn) "FTColorGrading" 0.3)))

(fn ConstructionScript
  (|CallParentFunction))
