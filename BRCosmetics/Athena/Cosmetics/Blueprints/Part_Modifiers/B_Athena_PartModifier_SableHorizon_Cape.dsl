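; Part modifier blueprint for the SableHorizon cape cosmetic, adjusting how the translucent back item sorts when rendered.
; On the post-res-in event, after the parent logic, it gets the skeletal mesh for the Backpack part on the associated player pawn.
; It applies a slight negative translucency sort distance offset to that mesh.
; The construction script only calls its parent.
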(event EventOnPostResIn
  (|CallParentFunction)
  (Rendering|SetTranslucencySortDistanceOffset (Customization|GetSkeletalMeshforPartType (CustomCharacter|GetAssociatedPlayerPawn) "Backpack" :!class /Script/FortniteGame.FortPlayerPawn) -0.1))

(fn ConstructionScript
  (|CallParentFunction))
