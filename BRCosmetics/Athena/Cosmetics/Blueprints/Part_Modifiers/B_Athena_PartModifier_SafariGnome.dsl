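; Cosmetic character part modifier for the Safari Gnome that timestamps the part's materials when it is displayed in the frontend.
; After res-in, if the frontend context check passes, it loops over the materials on the part's skeletal mesh.
; It casts each to a dynamic material instance and writes the current game time in seconds into the TOD_Timestamp scalar parameter.
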(event EventOnPostResIn
  (|CallParentFunction)
  (if (Reactive|Frontend|IsCosmeticObjectDisplayedinFrontendContext 0)
    (for _array_element (Rendering|Material|GetMaterials (CustomCharacter|GetSkeletalMeshForPartOnPlayerPawn))
      (bind _asmaterial_instance_dynamic (Utilities|Casting|CastToMaterialInstanceDynamic _array_element))
      (Rendering|Material|SetScalarParameterValue _asmaterial_instance_dynamic "TOD_Timestamp" (Utilities|Time|GetGameTimeinSeconds) :!class /Script/Engine.MaterialInstanceDynamic))))

(fn ConstructionScript
  (|CallParentFunction))
