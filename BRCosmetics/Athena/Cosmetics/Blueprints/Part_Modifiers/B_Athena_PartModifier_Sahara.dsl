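; Athena cosmetic part modifier (Sahara) that reveals the character part's material in levels as the player gets kills.
; After init it watches the AthenaKills stat through a gate that briefly closes to debounce changes.
; Each level-up, capped at the level count, stamps TransitionTime_FromBP and animates Reveal_Progress by timeline.
; In the front end, a looping preview timeline triggers level-ups, then resets and fades out before replaying.
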(event EventOnPostResIn
  (|CallParentFunction)
  (Utilities|FlowControl|Sequence
    (:then_0
      (bind (_statname _statvalue) (Cosmetics|WatchCosmeticStat (CustomCharacter|GetAssociatedPlayerPawn) "(TagName=\"StatManager.AthenaKills\")")
        (:OnStatChanged
          (if (> _statvalue 0)
            (Utilities|FlowControl|Gate false :!label _label
              (:Exit
                (Utilities|FlowControl|Sequence
                  (:then_0
                    _)
                  (:then_1
                    (Utilities|FlowControl|Sequence
                      (:then_0
                        (goto _label Close))
                      (:then_1
                        (Utilities|FlowControl|Delay 0.1)
                        (goto _label Open)))))))))))
    (:then_1
      (if (CustomCosmeticModifier|IsinFrontEnd)
        (Timeline|FE_Preview PlayFromStart
          (:Finished
            (Variables|Default|SetLevel 0)
            ;[1 1 1] FORT-566156: added fade out for FE preview
            (bind (_direction _fractionalprogress) (Timeline|Progress%20Fade%20out PlayFromStart)
              (:Update
                (Rendering|Material|SetScalarParameterValueonMaterials (CustomCharacter|GetSkeletalMeshForPartOnPlayerPawn) "Reveal_Progress" _fractionalprogress)
                (Timeline|FE_Preview PlayFromStart))))
          (:Level_Up
            _)))))
  (Rendering|Material|SetScalarParameterValueonMaterials (CustomCharacter|GetSkeletalMeshForPartOnPlayerPawn) "TransitionTime_FromBP" (Utilities|Time|GetGameTimeinSeconds))
  (if (< (Variables|Default|GetLevel) (Variables|Default|GetNumberOfLevels))
    (bind _result (Math|Integer|IncrementInt (Variables|Default|GetLevel))
      (:%20%20
        (bind (_direction_2 _fractionalprogress_2) (Timeline|Progress PlayFromStart)
          (:Update
            (Rendering|Material|SetScalarParameterValueonMaterials (CustomCharacter|GetSkeletalMeshForPartOnPlayerPawn) "Reveal_Progress" (- (/ (Math|Conversions|ToFloat(Integer) (Math|Integer|Min(Integer) _result (Variables|Default|GetNumberOfLevels))) (Math|Conversions|ToFloat(Integer) (Variables|Default|GetNumberOfLevels))) (* _fractionalprogress_2 (/ 1.0 (Math|Conversions|ToFloat(Integer) (Variables|Default|GetNumberOfLevels))))))))))))

(event EventOnDayPhaseChanged
  (params (CurrentDayPhase :type EFortDayPhase%20Enum :default "Morning") (PreviousDayPhase :type EFortDayPhase%20Enum :default "Morning") (bAtCreation :type Boolean)))

(event EventOnNewPlayerWeaponType
  (params (NewType :type EFortItemType%20Enum :default "WorldItem") (OldType :type EFortItemType%20Enum :default "WorldItem")))

(event EventOnWeaponChanged
  (params (NewWeapon :type Fort%20Weapon%20Object%20Reference) (OldWeapon :type Fort%20Weapon%20Object%20Reference)))

(event EventOnWeaponFired)

(event EventOnNewWeapons
  (params (WeaponsAsItems :type Array%20of%20Fort%20Item%20Object%20References)))

(event EventOnCurieStateAttached
  (params (Tag :type Gameplay%20Tag%20Structure%20%28by%20ref%29)))

(event EventOnCurieStateDetached
  (params (Tag :type Gameplay%20Tag%20Structure%20%28by%20ref%29)))

(event EventOnBeginSkydiving)

(event EventOnEndSkydiving)

(event EventOnEnterVehicle
  (params (VehicleEntered :type Fort%20Vehicle%20Interface%20Interface%20%28by%20ref%29)))

(event EventOnExitVehicle)

(event EventOnTargetingOrFirstPersonCameraChanged)

(event EventOnHitPawn
  (params (SelfActor :type Actor%20Object%20Reference) (OtherActor :type Actor%20Object%20Reference) (NormalImpulse :type Vector :default "0, 0, 0") (Hit :type Hit%20Result%20Structure%20%28by%20ref%29)))

(fn ConstructionScript)
