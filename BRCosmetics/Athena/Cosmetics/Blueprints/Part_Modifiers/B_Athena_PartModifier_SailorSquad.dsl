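; Cosmetic part modifier Blueprint for the SailorSquad character that tunes its material outline thickness depending on whether it is shown in the front end.
; On post res-in, it gets the skeletal mesh for its part type from the associated player pawn.
; It then sets the CRTNO outline inner-size scalar on that mesh's materials, using a lower value in the front end than in gameplay.
; The part-type switches carry no Head, Body or Face-specific logic.
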(event EventOnPostResIn
  (bind _returnvalue (CustomCharacter|GetAssociatedPlayerPawn))
  (bind _returnvalue_2 (CustomCharacter|GetCharacterPartType))
  (bind _returnvalue_3 (Customization|GetSkeletalMeshforPartType _returnvalue _returnvalue_2 :!class /Script/FortniteGame.FortPlayerPawn))
  (|CallParentFunction)
  (if (CustomCosmeticModifier|IsinFrontEnd)
    (switch Utilities|FlowControl|Switch|SwitchonEFortCustomPartType _returnvalue_2
      (:Head
        _)
      (:Body
        _)
      (:Face
        _))
    (Rendering|Material|SetScalarParameterValueonMaterials _returnvalue_3 "CRTNO_Outline-InnerSize1" 0.85)
    (else
      (switch Utilities|FlowControl|Switch|SwitchonEFortCustomPartType _returnvalue_2
        (:Head
          _)
        (:Body
          _)
        (:Face
          _))
      (Rendering|Material|SetScalarParameterValueonMaterials _returnvalue_3 "CRTNO_Outline-InnerSize1" 1.0))))

(fn ConstructionScript
  (|CallParentFunction))
