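; Character part modifier for the SameRabbit skin that plays a reactive glow-style effect on its material and idle VFX.
; On PostResIn it triggers the effect immediately in frontend display contexts, otherwise whenever the pawn's OpenChests stat changes.
; If the character's meta tags include ReactiveOn, a timeline drives the mesh material's EnableEffect and the particle EnableVFX parameters.
; The frontend Loop flag only leads to a trailing delay, with no replay wired after it.
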(event EventOnPostResIn
  (|CallParentFunction)
  (if (Reactive|Frontend|IsCosmeticObjectDisplayedinFrontendContext 0)
    (CallFunction|ReactivityEffect true)
    (else
      (Cosmetics|WatchCosmeticStat (CustomCharacter|GetAssociatedPlayerPawn) "(TagName=\"StatManager.OpenChests\")" 0 false
        (:OnStatChanged
          (CallFunction|ReactivityEffect false))))))

(event Custom|ReactivityEffect
  (params (Loop :type Boolean))
  (bind _returnvalue (CustomCharacter|GetSkeletalMeshForPartOnPlayerPawn))
  (if (GameplayTags|HasTag (Cosmetics|GatherMetaTags (CustomCharacter|GetAssociatedPlayerPawn) "/BRCosmetics/Athena/Items/Cosmetics/Characters/Character_SameRabbit.Character_SameRabbit" :!class /Script/FortniteGame.FortPlayerPawn) "(TagName=\"Cosmetics.MetaTags.ReactiveOn\")")
    (Utilities|IsValid _returnvalue
      (:Is%20Valid
        (bind (_direction _newtrack_0) (Timeline|Timeline PlayFromStart)
          (:Update
            (Rendering|Material|SetScalarParameterValue (Class|BTechArtGlobalLibrary|GetDynamicMaterialInstance _returnvalue) "EnableEffect" _newtrack_0 :!class /Script/Engine.MaterialInstanceDynamic)
            (Effects|Components|ParticleSystem|SetFloatParameter (CallFunction|GetIdleVFXComponent) "EnableVFX" _newtrack_0))
          (:Finished
            (if Loop
              (Utilities|FlowControl|Delay 5.0))))))))

(fn ConstructionScript
  (|CallParentFunction))
