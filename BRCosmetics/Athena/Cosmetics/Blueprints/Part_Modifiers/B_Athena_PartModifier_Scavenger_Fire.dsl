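; Cosmetic part modifier for the Scavenger Fire character part, adding fire effects to the player's feet and body.
; After the post-res-in event runs its parent logic, it attaches left and right heel fire components to the foot bones and fire trails to the root of the part's skeletal mesh.
; All three are registered in the additional FX list that needs hidden-part checks.
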(event EventOnPostResIn
  (bind _additional_fx_needing_part_checks (Variables|HiddenPartCheck|GetAdditional%20FX%20Needing%20Part%20Checks))
  (|CallParentFunction)
  (Transformation|AttachComponentToComponent (Variables|Default|GetHeelFire_L) (CustomCharacter|GetSkeletalMeshForPartOnPlayerPawn) "foot_l" "KeepRelative" "KeepRelative" "KeepRelative" true)
  (Transformation|AttachComponentToComponent (Variables|Default|GetHeelFire_R) (CustomCharacter|GetSkeletalMeshForPartOnPlayerPawn) "foot_r" "KeepRelative" "KeepRelative" "KeepRelative" true)
  (Transformation|AttachComponentToComponent (Variables|Default|GetFireTrails) (CustomCharacter|GetSkeletalMeshForPartOnPlayerPawn) "root" "KeepRelative" "KeepRelative" "KeepRelative" true)
  (Utilities|Array|Add _additional_fx_needing_part_checks (Variables|Default|GetHeelFire_L))
  (Utilities|Array|Add _additional_fx_needing_part_checks (Variables|Default|GetHeelFire_R))
  (Utilities|Array|Add _additional_fx_needing_part_checks (Variables|Default|GetFireTrails)))

(fn ConstructionScript
  (|CallParentFunction))
