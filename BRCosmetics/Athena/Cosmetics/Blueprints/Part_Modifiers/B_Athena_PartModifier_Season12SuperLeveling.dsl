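; Season 12 super-leveling cosmetic part modifier that drives a character part's materials from the player's season level.
; On post-res-in it watches the SeasonLevel12 stat, keeps the highest value, and pushes level range, radius and current level as material scalars.
; A looping 0.1s workaround timer reapplies these parameters and is cleared after a one-second delay.
; A gated range check calls OnActualStatChange, which stamps game time into the materials.
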(event Custom|HackyWorkaroundRefreshTimerGuy
  ;[1 1 1] SuperLevelling Logic
  (Utilities|FlowControl|Sequence
    (:then_0
      (CallFunction|InitializeVariables)
      (Rendering|Material|SetScalarParameterValueonMaterials (CustomCharacter|GetSkeletalMeshForPartOnPlayerPawn) "SL_CurrentLevel" (Math|Conversions|ToFloat(Integer) (Variables|Default|GetSL_LastValue)))
      (Utilities|FlowControl|DoOnce false
        (:Completed
          (bind _returnvalue (Utilities|Time|SetTimerbyEvent (EventDispatchers|CreateEvent :!event Custom|HackyWorkaroundRefreshTimerGuy) 0.1 true))
          (Variables|Default|SetSL_RefreshTimeHandler _returnvalue))))
    (:then_1
      (Utilities|FlowControl|Gate true
        (:Exit
          (if (and (> (Math|Integer|Max(Integer) (stale EventOnPostResIn _statvalue) (Variables|Default|GetSL_LastValue)) (Variables|Default|GetSL_StartLevel)) (< (Math|Integer|Max(Integer) (stale EventOnPostResIn _statvalue) (Variables|Default|GetSL_LastValue)) (Variables|Default|GetSL_EndLevel)))
            (CallFunction|OnActualStatChange)))))))

(event EventOnPostResIn
  (bind _returnvalue (CustomCharacter|GetSkeletalMeshForPartOnPlayerPawn))
  (bind _sl_lastvalue (Variables|Default|GetSL_LastValue))
  (|CallParentFunction)
  (CallFunction|ResetSafetyGate)
  ;[1 1 1] SuperLevelling Logic
  (Utilities|FlowControl|Sequence
    (:then_0
      (Utilities|FlowControl|Delay 1.0)
      (Utilities|Time|ClearandInvalidateTimerbyHandle (Variables|Default|GetSL_RefreshTimeHandler)))
    (:then_1
      (bind (_statname _statvalue) (Cosmetics|WatchCosmeticStat (CustomCharacter|GetAssociatedPlayerPawn) "(TagName=\"StatManager.Season.SeasonLevel12\")")
        (:OnStatChanged
          (Rendering|Material|SetScalarParameterValueonMaterials _returnvalue "SL_StartAtLevel" (Math|Conversions|ToFloat(Integer) (Variables|Default|GetSL_StartLevel)))
          (Rendering|Material|SetScalarParameterValueonMaterials _returnvalue "SL_StopAtLevel" (Math|Conversions|ToFloat(Integer) (Variables|Default|GetSL_EndLevel)))
          (Rendering|Material|SetScalarParameterValueonMaterials _returnvalue "SL_EndRadius" (Variables|Default|GetEndRadius))
          (Variables|Default|SetSL_LastValue (Math|Integer|Max(Integer) _statvalue _sl_lastvalue))
          (Utilities|FlowControl|Sequence
            (:then_0
              (CallFunction|InitializeVariables)
              (Rendering|Material|SetScalarParameterValueonMaterials (CustomCharacter|GetSkeletalMeshForPartOnPlayerPawn) "SL_CurrentLevel" (Math|Conversions|ToFloat(Integer) (Variables|Default|GetSL_LastValue)))
              (Utilities|FlowControl|DoOnce false
                (:Completed
                  (bind _returnvalue_2 (Utilities|Time|SetTimerbyEvent (EventDispatchers|CreateEvent :!event Custom|HackyWorkaroundRefreshTimerGuy) 0.1 true))
                  (Variables|Default|SetSL_RefreshTimeHandler _returnvalue_2))))
            (:then_1
              (Utilities|FlowControl|Gate true
                (:Exit
                  (if (and (> (Math|Integer|Max(Integer) _statvalue _sl_lastvalue) (Variables|Default|GetSL_StartLevel)) (< (Math|Integer|Max(Integer) _statvalue _sl_lastvalue) (Variables|Default|GetSL_EndLevel)))
                    (CallFunction|OnActualStatChange)))))))))))

(event Custom|ResetSafetyGate
  ;[1 1 1] SuperLevelling Logic
  (Utilities|FlowControl|DoOnce false
    (:Completed
      (bind _returnvalue (Utilities|Time|SetTimerbyEvent (EventDispatchers|CreateEvent :!event Custom|HackyWorkaroundRefreshTimerGuy) 0.1 true))
      (Variables|Default|SetSL_RefreshTimeHandler _returnvalue))))

(event Custom|OnActualStatChange
  ;[1 1 1] SuperLevelling Logic
  ;;[1 1 1] This should only fire off when the stat changes - use this event!
  (Rendering|Material|SetScalarParameterValueonMaterials (CustomCharacter|GetSkeletalMeshForPartOnPlayerPawn) "SL_GameTimeFromBP" (Utilities|Time|GetGameTimeinSeconds)))

(fn GetAnimInstance
  (return (Components|SkeletalMesh|GetAnimInstance (Customization|GetSkeletalMeshforPartType (CustomCharacter|GetAssociatedPlayerPawn) (CustomCharacter|GetCharacterPartType) :!class /Script/FortniteGame.FortPlayerPawn))))

(fn InitializeVariables
  (bind _returnvalue (CustomCharacter|GetSkeletalMeshForPartOnPlayerPawn))
  (Rendering|Material|SetScalarParameterValueonMaterials _returnvalue "SL_StartAtLevel" (Math|Conversions|ToFloat(Integer) (Variables|Default|GetSL_StartLevel)))
  (Rendering|Material|SetScalarParameterValueonMaterials _returnvalue "SL_StopAtLevel" (Math|Conversions|ToFloat(Integer) (Variables|Default|GetSL_EndLevel)))
  (Rendering|Material|SetScalarParameterValueonMaterials _returnvalue "SL_EndRadius" (Variables|Default|GetEndRadius)))

(fn ConstructionScript
  (|CallParentFunction))
