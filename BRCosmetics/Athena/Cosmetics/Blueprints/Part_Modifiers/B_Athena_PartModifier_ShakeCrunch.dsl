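; Cosmetic part modifier for the ShakeCrunch character that makes its materials react to gameplay.
; After post-resurrection init, a looping timer randomizes the collage outline color on MiscOrTail, with brightness depending on frontend context.
; Stat watchers for kills, chests, shield, health and damage trigger a random glitch color pair and stamp game time into Reactive_Main_CT.
; An equipped ExtraPart sets UseShoes to hide the bodybag feet outline (FORT-868965).
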(event Custom|FE_Reactivety)

(event Custom|ChangeCollageColor
  ;[1 0.8396 0.3695] Collage Outline color change
  (CallFunction|SetRandomCollageColors))

(event Custom|OnKill
  ;[0.3423 1 0.4786] Events
  (CallFunction|SetMainColors))

(event EventOnPostResIn
  (|CallParentFunction)
  (Utilities|ToonFrontendLUTReplace
    (:Out
      (Utilities|FlowControl|Sequence
        (:then_0
          ;[1 0.8396 0.3695] Collage Outline color change
          (Utilities|Time|SetTimerbyEvent (delegate Custom|ChangeCollageColor) 3.3 true false 6.0 1.3))
        (:then_1
          ;[0.2818 1 0.7667] On Kill
          (Cosmetics|WatchCosmeticStat (CustomCharacter|GetAssociatedPlayerPawn) "(TagName=\"StatManager.AthenaKills\")" 0 false
            (:OnStatChanged
              (CallFunction|OnKill))))
        (:then_2
          ;[1 0.9986 0.1692] On Open Chest
          (Cosmetics|WatchCosmeticStat (CustomCharacter|GetAssociatedPlayerPawn) "(TagName=\"StatManager.OpenChests\")" 0 false
            (:OnStatChanged
              (CallFunction|OnOpenChest))))
        (:then_3
          ;[1 0.2708 0.2708] On Gained Shield
          (Cosmetics|WatchCosmeticStat (CustomCharacter|GetAssociatedPlayerPawn) "(TagName=\"StatManager.GainedShieldTimes\")" 0 false
            (:OnStatChanged
              (CallFunction|OnGainedShield))))
        (:then_4
          ;[1 0.6724 0.2216] On Gained Health
          (Cosmetics|WatchCosmeticStat (CustomCharacter|GetAssociatedPlayerPawn) "(TagName=\"StatManager.GainedHealthTimes\")" 0 false
            (:OnStatChanged
              (CallFunction|OnGainedHealth))))
        (:then_5
          ;[1 0.2219 0.2219] On Damage Received
          (Cosmetics|WatchCosmeticStat (CustomCharacter|GetAssociatedPlayerPawn) "(TagName=\"StatManager.DamageReceived\")" 0 false
            (:OnStatChanged
              (CallFunction|OnDamageReceived))))
        (:then_6
          ; FORT-868965 Hide bodybag feet outline when shoes on
          (if (Utilities|IsValid (CosmeticLoadout|GetItemDefinitionfromLoadout (CosmeticLoadout|GetCosmeticLoadout) "(TagName=\"LoadoutSlot.ExtraPart\")") :!class /Script/Engine.KismetSystemLibrary)
            (Rendering|Material|SetScalarParameterValue (Class|BTechArtGlobalLibrary|GetDynamicMaterialInstance (Customization|GetSkeletalMeshforPartType (CustomCharacter|GetAssociatedPlayerPawn) "MiscOrTail" :!class /Script/FortniteGame.FortPlayerPawn)) "UseShoes" 1.0 :!class /Script/Engine.MaterialInstanceDynamic)))))))

(event Custom|OnOpenChest
  ;[0.3423 1 0.4786] Events
  (CallFunction|SetMainColors))

(event Custom|OnGainedShield
  ;[0.3423 1 0.4786] Events
  (CallFunction|SetMainColors))

(event Custom|OnGainedHealth
  ;[0.3423 1 0.4786] Events
  (CallFunction|SetMainColors))

(event Custom|OnDamageReceived
  ;[0.3423 1 0.4786] Events
  (CallFunction|SetMainColors))

(fn SetMainColors
  (bind _id (Variables|Default|GetID))
  (bind _maincolor (Variables|Default|GetMainColor))
  (bind _output (Utilities|Array|Get(aref) _maincolor _id))
  (bind (_primarycolor_5_cacd3386420963b23cccbfa04cc5b2d0 _secondarycolor_7_b40e99f44f6db07d694f5cb8bb2959ef) (Utilities|Struct|BreakSTRShakeCrunchMainColors _output))
  (bind _returnvalue (CustomCharacter|GetAssociatedPlayerPawn))
  (Variables|Default|SetID (Math|Random|RandomIntegerinRange 0 9))
  (Rendering|Material|SetColorParameterValueonMaterials (Customization|GetSkeletalMeshforPartType _returnvalue "Face" :!class /Script/FortniteGame.FortPlayerPawn) "GlitchColor_Primary" (Math|Conversions|ToLinearColor(Color) _primarycolor_5_cacd3386420963b23cccbfa04cc5b2d0))
  (Rendering|Material|SetColorParameterValueonMaterials (Customization|GetSkeletalMeshforPartType _returnvalue "Head" :!class /Script/FortniteGame.FortPlayerPawn) "GlitchColor_Secondary" (Math|Conversions|ToLinearColor(Color) _secondarycolor_7_b40e99f44f6db07d694f5cb8bb2959ef))
  (Rendering|Material|SetScalarParameterValueonMaterials (CustomCharacter|GetSkeletalMeshForPartOnPlayerPawn) "Reactive_Main_CT" (Utilities|Time|GetGameTimeinSeconds)))

(fn SetRandomCollageColors
  (bind _element (ArrayMacros|GetRandomArrayElement (Variables|Default|GetCollageColor))
    (:%20%20
      (Rendering|Material|SetColorParameterValueonMaterials (Customization|GetSkeletalMeshforPartType (CustomCharacter|GetAssociatedPlayerPawn) "MiscOrTail" :!class /Script/FortniteGame.FortPlayerPawn) "Collage_Color" (* (Math|Conversions|ToLinearColor(Color) _element) (select (Reactive|Frontend|IsCosmeticObjectDisplayedinFrontendContext 0) "(R=0.700000,G=0.700000,B=0.700000,A=1.000000)" "(R=1.000000,G=1.000000,B=1.000000,A=1.000000)"))))))

(fn ConstructionScript
  (|CallParentFunction))
