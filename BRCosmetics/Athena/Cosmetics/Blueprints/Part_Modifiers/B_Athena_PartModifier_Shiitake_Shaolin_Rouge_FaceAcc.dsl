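; Cosmetic part modifier for the Shiitake Shaolin Rouge face accessory that drives its particle effects from player pawn events.
; On landing, it spawns the jump FX attached to the pawn's part mesh at FX_Cap and passes in whether the head is hidden.
; Vehicle and skydiving events toggle matching bool parameters on the idle VFX component, with a short delay for skydiving changes.
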(event EventOnPawnLanded
  (|CallParentFunction)
  (bind _returnvalue (PartModifier|SpawnSystem|SpawnSystemAttachedPartModifier (Variables|Default|GetJump%20FX) (CustomCharacter|GetSkeletalMeshForPartOnPlayerPawn) "FX_Cap")
    (:Success
      (Effects|Components|ParticleSystem|SetBoolParameter _returnvalue "bisHeadHidden" (Variables|HiddenPartCheck|GetbIsHeadHidden)))))

(event EventOnEnterVehicle
  (params (VehicleEntered :type Fort%20Vehicle%20Interface%20Interface%20%28by%20ref%29))
  (|CallParentFunction VehicleEntered)
  (Effects|Components|ParticleSystem|SetBoolParameter (CallFunction|GetIdleVFXComponent) "IsDriving" true))

(event EventOnExitVehicle
  (|CallParentFunction)
  (Effects|Components|ParticleSystem|SetBoolParameter (CallFunction|GetIdleVFXComponent) "IsDriving" false))

(event EventOnBeginSkydiving
  (|CallParentFunction)
  (Utilities|FlowControl|Delay 0.1)
  (Effects|Components|ParticleSystem|SetBoolParameter (CallFunction|GetIdleVFXComponent) "IsSkydiving" true))

(event EventOnEndSkydiving
  (|CallParentFunction)
  (Utilities|FlowControl|Delay 0.1)
  (Effects|Components|ParticleSystem|SetBoolParameter (CallFunction|GetIdleVFXComponent) "IsSkydiving" false))

(fn ConstructionScript)
