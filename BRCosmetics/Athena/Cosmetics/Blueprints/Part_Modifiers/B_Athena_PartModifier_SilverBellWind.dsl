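; Character part modifier for the SilverBellWind cosmetic, driving a Niagara effect attached to the character part's skeletal mesh at Hair_MSTR.
; Construction and post-resurrection-in setup pass three spline components to Niagara as object references, and construction enables debug drawing on the first.
; Post-res-in also rotates and attaches the effect, then feeds it the active Mat1/Mat2 material variant index and a front-end preview flag.
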(event EventOnPostResIn
  (|CallParentFunction)
  (CallFunction|SplineReferencesSetter)
  (CallFunction|AttachComponent "Hair_MSTR")
  (bind _variantindex (CallFunction|GetVariantIndex))
  (CallFunction|SetNiagaraVariables (Math|Conversions|ToFloat(Integer) _variantindex) (CustomCosmeticModifier|IsinFrontEnd)))

(event EventOnPawnUpdated
  (|CallParentFunction))

(event OnComponentSleep(Niagara)
  (params (SleepingComponent :type Primitive%20Component%20Object%20Reference) (BoneName :type Name :default "None")))

(fn AttachComponent
  (params (SocketName :type Name))
  (bind _niagara (Variables|Default|GetNiagara))
  (Transformation|SetRelativeRotation (Variables|Default|GetNiagara) "-90.000000,0.000000,0.000000" false false) ; idk why
  (Utilities|IsValid _niagara
    (:Is%20Valid
      (Transformation|AttachComponentToComponent _niagara (CustomCharacter|GetSkeletalMeshForPartOnPlayerPawn) SocketName "KeepRelative" "KeepRelative" "KeepRelative" true))))

(fn GetVariantIndex
  (bind _asfort_player_pawn (Utilities|Casting|CastToFortPlayerPawn (Actor|GetOwner)))
  (bind _index (CosmeticLoadout|GetVariantChannelActiveIndexforSlotActor "(TagName=\"LoadoutSlot.Character\")" _asfort_player_pawn "(TagName=\"Cosmetics.Variant.Channel.Material\")" (Utilities|Array|MakeArray "(TagName=\"Cosmetics.Variant.Property.Mat1\")" "(TagName=\"Cosmetics.Variant.Property.Mat2\")")))
  (bind _output_get (Variables|Default|SetVariantIndex _index))
  (return _output_get))

(fn SetNiagaraVariables
  (params (Variant :type Float%20%28single-precision%29) (isFrontend :type Boolean))
  (Effects|Components|ParticleSystem|SetFloatParameter (Variables|Default|GetNiagara) "Variant" Variant)
  (Effects|Components|ParticleSystem|SetBoolParameter (Variables|Default|GetNiagara) "bisFrontEndPreview" isFrontend))

(fn SplineReferencesSetter
  (Niagara|SetNiagaraVariable(Object) (Variables|Default|GetNiagara) "SplineReference" (Variables|Default|GetSpline))
  (Niagara|SetNiagaraVariable(Object) (Variables|Default|GetNiagara) "SplineReference2" (Variables|Default|GetSpline1))
  (Niagara|SetNiagaraVariable(Object) (Variables|Default|GetNiagara) "SplineReference3" (Variables|Default|GetSpline2)))

(fn ConstructionScript
  (|CallParentFunction)
  (CallFunction|SplineReferencesSetter)
  (Spline|SetDrawDebug (Variables|Default|GetSpline) true))
