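; Part modifier Blueprint for the Skirmisher Male cosmetic character part.
; When the pawn lands, it gets the idle VFX component, casts it to a Niagara component and resets that particle system.
; The construction script only calls the parent implementation.
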(event EventOnPawnLanded
  (bind _asniagara_particle_system_component (Utilities|Casting|CastToNiagaraComponent (CallFunction|GetIdleVFXComponent)))
  (Niagara|ResetSystem _asniagara_particle_system_component))

(fn ConstructionScript
  (|CallParentFunction))
