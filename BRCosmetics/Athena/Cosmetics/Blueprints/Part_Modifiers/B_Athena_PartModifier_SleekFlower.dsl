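; Part modifier for the Sleek outfit that makes its flowers react to eliminations.
; Stat changes set material stage, count and game-time parameters, spawn a Niagara hit effect, and toggle an idle aura at max eliminations.
; Beyond that maximum, the emissive switches to looping.
; In the front end, the FE_Preview timeline drives the same stage parameters plus glow, spawns DoOnce-gated effects and calls a reactivity loop function on finish.
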(event Custom|FE_ReactivityLoop
  (bind _returnvalue (CustomCharacter|GetSkeletalMeshForPartOnPlayerPawn))
  ;[1 1 1] FrontEnd Reactivity Preview
  (bind (_direction _newtrack_0 _stage_01 _stage_02 _stage_03 _stage_04 _stage_05 _emissivetime) (Timeline|FE_Preview PlayFromStart)
    (:Update
      (Rendering|Material|SetScalarParameterValueonMaterials _returnvalue "SF_Stage_01" _stage_01)
      (Rendering|Material|SetScalarParameterValueonMaterials _returnvalue "SF_Stage_02" _stage_02)
      (Rendering|Material|SetScalarParameterValueonMaterials _returnvalue "SF_Stage_03" _stage_03)
      (Rendering|Material|SetScalarParameterValueonMaterials _returnvalue "SF_Stage_04" _stage_04)
      (Rendering|Material|SetScalarParameterValueonMaterials _returnvalue "SF_Stage_05" 1.0)
      (Rendering|Material|SetScalarParameterValueonMaterials _returnvalue "SF_Stage_05" _stage_05)
      (Rendering|Material|SetScalarParameterValueonMaterials _returnvalue "TransformControl" _emissivetime)
      (Rendering|Material|SetScalarParameterValueonMaterials _returnvalue "SF_FrontEnd_GlowTIme" _newtrack_0)
      (Utilities|FlowControl|Sequence
        (:then_0
          (if (Math|Boolean|ORBoolean (== _newtrack_0 0.2) (== _newtrack_0 0.4) (== _newtrack_0 0.6) (== _newtrack_0 0.8) (>= _newtrack_0 1.0))
            _)
          (Utilities|FlowControl|DoOnce false
            (:Completed
              (bind _returnvalue_2 (PartModifier|SpawnSystem|SpawnSystemAttachedPartModifier "/BRCosmetics/Characters/Player/Male/Medium/Bodies/M_MED_Sleek_Glasses_Jacket/FX/NS_Sleek_Body_FlowerHit.NS_Sleek_Body_FlowerHit" (CustomCharacter|GetSkeletalMeshForPartOnPlayerPawn) "root" "0, 0, 0" "0, 0, 0" "KeepRelativeOffset" true false)
                (:Success
                  (Utilities|FlowControl|RetriggerableDelay 1.0)
                  (Components|Activation|Activate _returnvalue_2 false))))))
        (:then_1
          ;[1 1 1] Comment
          (if (>= _stage_05 0.8)
            _)
          (Utilities|FlowControl|DoOnce false
            (:Completed
              (bind _returnvalue_3 (PartModifier|SpawnSystem|SpawnSystemAttachedPartModifier "/BRCosmetics/Characters/Player/Male/Medium/Bodies/M_MED_Sleek_Glasses_Jacket/FX/NS_Sleek_Body_Energy_FrontEnd.NS_Sleek_Body_Energy_FrontEnd" (CustomCharacter|GetSkeletalMeshForPartOnPlayerPawn) "root" "0, 0, 0" "0, 0, 0" "KeepRelativeOffset" true false)
                (:Success
                  (Components|Activation|Activate _returnvalue_3 false))))))))
    (:Finished
      (CallFunction|FEReactivityLoop))))

(event EventOnPostResIn
  (bind _returnvalue (CustomCharacter|GetSkeletalMeshForPartOnPlayerPawn))
  ;[1 1 1] FrontEnd Reactivity Preview
  (|CallParentFunction)
  (if (CustomCosmeticModifier|IsinFrontEnd)
    (Rendering|Material|SetScalarParameterValueonMaterials _returnvalue "UseParameterValueForFlowers" 1.0)
    (Timeline|FE_Preview PlayFromStart)
    (else
      (Rendering|Material|SetScalarParameterValueonMaterials _returnvalue "UseParameterValueForFlowers" 0.0))))

(event EventOnStatChanged
  (params (State%20Name :type Name :default "None") (Stat%20Value :type Integer))
  (bind _returnvalue (CustomCharacter|GetSkeletalMeshForPartOnPlayerPawn))
  (bind _returnvalue_2 (Utilities|Time|GetGameTimeinSeconds))
  (bind _returnvalue_3 (CustomCharacter|GetSkeletalMeshForPartOnPlayerPawn))
  (bind _returnvalue_4 (CallFunction|GetIdleVFXComponent))
  (bind _returnvalue_5 (< Stat%20Value 1))
  (|CallParentFunction State%20Name Stat%20Value)
  (Utilities|FlowControl|Sequence
    (:then_0
      ;[1 1 1] Set Mesh Materials
      (switch int (Variables|PartModifierControls|WatchCosmeticStat|GetCurrent%20Active%20Stage%20Number)
        (:Default
          _)
        (:0
          _)
        (:1
          (Rendering|Material|SetScalarParameterValueonMaterials _returnvalue_3 "SF_Stage_01" 1.0))
        (:2
          (Rendering|Material|SetScalarParameterValueonMaterials _returnvalue_3 "SF_Stage_02" 1.0))
        (:3
          (Rendering|Material|SetScalarParameterValueonMaterials _returnvalue_3 "SF_Stage_03" 1.0))
        (:4
          (Rendering|Material|SetScalarParameterValueonMaterials _returnvalue_3 "SF_Stage_04" 1.0))
        (:5
          (Rendering|Material|SetScalarParameterValueonMaterials _returnvalue_3 "SF_Stage_05" 1.0)))
      (Rendering|Material|SetScalarParameterValueonMaterials _returnvalue_3 "SF_Stage_01" 0.0))
    (:then_1
      (if _returnvalue_5
        _
        (elif (> Stat%20Value (Variables|Default|GetMaxElims))
          (Rendering|Material|SetScalarParameterValueonMaterials _returnvalue "SF_SwitchtoLoopingEmissive" 1.0)
          (Rendering|Material|SetScalarParameterValueonMaterials _returnvalue "SF_GameTimeFromBPEmissive" _returnvalue_2)
          (else
            (Rendering|Material|SetScalarParameterValueonMaterials _returnvalue "SF_SwitchtoLoopingEmissive" 0.0)
            (Rendering|Material|SetScalarParameterValueonMaterials _returnvalue "SF_GameTimeFromBPEmissive" _returnvalue_2)
            (Rendering|Material|SetScalarParameterValueonMaterials _returnvalue "SF_GameTimeFromBP" _returnvalue_2)
            (Rendering|Material|SetScalarParameterValueonMaterials _returnvalue "SF_CurrentElims" (Math|Conversions|ToFloat(Integer) Stat%20Value))))))
    (:then_2
      (if (not _returnvalue_5)
        ;[1 1 1] Reactive
        (bind _returnvalue_6 (PartModifier|SpawnSystem|SpawnSystemAttachedPartModifier "/BRCosmetics/Characters/Player/Male/Medium/Bodies/M_MED_Sleek_Glasses_Jacket/FX/NS_Sleek_Body_FlowerHit.NS_Sleek_Body_FlowerHit" (CustomCharacter|GetSkeletalMeshForPartOnPlayerPawn) "root" "0, 0, 0" "0, 0, 0" "KeepRelativeOffset" true false)
          (:Success
            (Components|Activation|Activate _returnvalue_6 false)))))
    (:then_3
      ;[1 1 1] Aura
      (if (>= Stat%20Value (Variables|Default|GetMaxElims))
        (Components|Activation|Activate _returnvalue_4 false)
        (else
          ;[1 1 1] Added for jira FORT-484999
          (Components|Activation|Deactivate _returnvalue_4))))))

(fn ConstructionScript)
