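; Backpack part modifier for the SleepyDuck cosmetic that turns on custom depth rendering for the player's backpack mesh.
; On post-resurrection-in event, it calls the parent, then briefly delays only when displayed in a frontend context.
; It then finds the Backpack skeletal mesh on the associated player pawn and enables render custom depth on it.
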(event EventOnPostResIn
  (|CallParentFunction)
  (if (Reactive|Frontend|IsCosmeticObjectDisplayedinFrontendContext 0)
    (Utilities|FlowControl|Delay 0.15))
  (Rendering|SetRenderCustomDepth (Customization|GetSkeletalMeshforPartType (CustomCharacter|GetAssociatedPlayerPawn) "Backpack" :!class /Script/FortniteGame.FortPlayerPawn) true))

(fn ConstructionScript
  (|CallParentFunction))
