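; Reactive character part modifier for the Snowfall Pink Ice cosmetic that evolves its materials and FX as eliminations rise.
; OnStatChanged maps the count to stages starting at 1, 3 and 6, toggling hidden elements and spawning attached Niagara bursts.
; Game-time scalar parameters trigger ice shard and frost animations, or a wave past six.
; In the frontend, OnPostResIn previews counts zero to six in one delayed sequence, using duplicated per-count events.
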(event Custom|Snowfall%200%20Emliminations
  (bind _returnvalue (CustomCharacter|GetSkeletalMeshForPartOnPlayerPawn))
  (bind _returnvalue_2 (Utilities|Time|GetGameTimeinSeconds))
  (bind _returnvalue_3 (CustomCharacter|GetSkeletalMeshForPartOnPlayerPawn))
  (bind _current_cycle (Variables|Default|GetCurrent%20Cycle))
  ;[0.007499 0.2122 0.6584] FrontEnd Stage cycling
  (Variables|Default|SetCurrent%20Cycle 0)
  (Utilities|FlowControl|Sequence
    (:then_0
      ;[0.272 0.8177 0.1576] Material Reactivity
      (Rendering|Material|SetScalarParameterValueonMaterials _returnvalue_3 "Hide Element 01" 1.0)
      (Rendering|Material|SetScalarParameterValueonMaterials _returnvalue_3 "Hide Element 02" 1.0)
      (Rendering|Material|SetScalarParameterValueonMaterials _returnvalue_3 "Hide Element 03" 1.0)
      (Rendering|Material|SetScalarParameterValueonMaterials _returnvalue_3 "Hide Element 10" 0.0)
      (Rendering|Material|SetScalarParameterValueonMaterials _returnvalue_3 "SP_Current_Stage" 0.0)
      (Rendering|Material|SetScalarParameterValueonMaterials _returnvalue_3 "SP_Current_Eliminations" 0.0))
    (:then_1
      (if (< _current_cycle 1)
        _
        (elif (> _current_cycle 6)
          (Rendering|Material|SetScalarParameterValueonMaterials _returnvalue "SP_CurrentTime_Wave" _returnvalue_2)
          (else
            (Rendering|Material|SetScalarParameterValueonMaterials _returnvalue "SP_Current_Eliminations" (Math|Conversions|ToFloat(Integer) (Variables|Default|GetCurrent%20Cycle)))
            (Rendering|Material|SetScalarParameterValueonMaterials _returnvalue "SP_CurrentTime_Frost" _returnvalue_2)))))))

(event EventOnPostResIn
  (bind _frontendfxcycledelay (Variables|Default|GetFrontendFXCycleDelay))
  ;[0.007499 0.2122 0.6584] FrontEnd Stage cycling
  (|CallParentFunction)
  (CallFunction|Snowfall0Emliminations)
  (if (CustomCosmeticModifier|IsinFrontEnd)
    (CallFunction|Snowfall0Emliminations)
    (Utilities|FlowControl|Delay _frontendfxcycledelay)
    (CallFunction|Snowfall1Emliminations)
    (Utilities|FlowControl|Delay _frontendfxcycledelay)
    (CallFunction|Snowfall2Emliminations)
    (Utilities|FlowControl|Delay _frontendfxcycledelay)
    (CallFunction|Snowfall3Emliminations)
    (Utilities|FlowControl|Delay _frontendfxcycledelay)
    (CallFunction|Snowfall4Emliminations)
    (Utilities|FlowControl|Delay _frontendfxcycledelay)
    (CallFunction|Snowfall5Emliminations)
    (Utilities|FlowControl|Delay _frontendfxcycledelay)
    (CallFunction|Snowfall6Emliminations)
    (Utilities|FlowControl|Delay _frontendfxcycledelay)))

(event Custom|Snowfall%201%20Emliminations
  (bind _returnvalue (CustomCharacter|GetSkeletalMeshForPartOnPlayerPawn))
  (bind _returnvalue_2 (Utilities|Time|GetGameTimeinSeconds))
  (bind _returnvalue_3 (CustomCharacter|GetSkeletalMeshForPartOnPlayerPawn))
  (bind _current_cycle (Variables|Default|GetCurrent%20Cycle))
  ;[0.007499 0.2122 0.6584] FrontEnd Stage cycling
  (Variables|Default|SetCurrent%20Cycle 1)
  (Utilities|FlowControl|Sequence
    (:then_0
      ;[0.272 0.8177 0.1576] Material Reactivity
      (Rendering|Material|SetScalarParameterValueonMaterials _returnvalue_3 "Hide Element 01" 0.0)
      (Rendering|Material|SetScalarParameterValueonMaterials _returnvalue_3 "SP_Current_Stage" 1.0)
      (CallFunction|Stage1FX)
      (Rendering|Material|SetScalarParameterValueonMaterials (CustomCharacter|GetSkeletalMeshForPartOnPlayerPawn) "SP_CurrentTime_IceShards" (Utilities|Time|GetGameTimeinSeconds)))
    (:then_1
      (if (< _current_cycle 1)
        _
        (elif (> _current_cycle 6)
          (Rendering|Material|SetScalarParameterValueonMaterials _returnvalue "SP_CurrentTime_Wave" _returnvalue_2)
          (else
            (Rendering|Material|SetScalarParameterValueonMaterials _returnvalue "SP_Current_Eliminations" (Math|Conversions|ToFloat(Integer) (Variables|Default|GetCurrent%20Cycle)))
            (Rendering|Material|SetScalarParameterValueonMaterials _returnvalue "SP_CurrentTime_Frost" _returnvalue_2)))))))

(event Custom|Snowfall%202%20Emliminations
  (bind _returnvalue (CustomCharacter|GetSkeletalMeshForPartOnPlayerPawn))
  (bind _returnvalue_2 (Utilities|Time|GetGameTimeinSeconds))
  (bind _returnvalue_3 (CustomCharacter|GetSkeletalMeshForPartOnPlayerPawn))
  (bind _current_cycle (Variables|Default|GetCurrent%20Cycle))
  ;[0.007499 0.2122 0.6584] FrontEnd Stage cycling
  (Variables|Default|SetCurrent%20Cycle 2)
  (Utilities|FlowControl|Sequence
    (:then_0
      ;[0.272 0.8177 0.1576] Material Reactivity
      (Rendering|Material|SetScalarParameterValueonMaterials _returnvalue_3 "Hide Element 01" 0.0)
      (Rendering|Material|SetScalarParameterValueonMaterials _returnvalue_3 "SP_Current_Stage" 1.0))
    (:then_1
      (if (< _current_cycle 1)
        _
        (elif (> _current_cycle 6)
          (Rendering|Material|SetScalarParameterValueonMaterials _returnvalue "SP_CurrentTime_Wave" _returnvalue_2)
          (else
            (Rendering|Material|SetScalarParameterValueonMaterials _returnvalue "SP_Current_Eliminations" (Math|Conversions|ToFloat(Integer) (Variables|Default|GetCurrent%20Cycle)))
            (Rendering|Material|SetScalarParameterValueonMaterials _returnvalue "SP_CurrentTime_Frost" _returnvalue_2)))))))

(event Custom|Snowfall%203%20Emliminations
  (bind _returnvalue (CustomCharacter|GetSkeletalMeshForPartOnPlayerPawn))
  (bind _returnvalue_2 (Utilities|Time|GetGameTimeinSeconds))
  (bind _returnvalue_3 (CustomCharacter|GetSkeletalMeshForPartOnPlayerPawn))
  (bind _current_cycle (Variables|Default|GetCurrent%20Cycle))
  ;[0.007499 0.2122 0.6584] FrontEnd Stage cycling
  (Variables|Default|SetCurrent%20Cycle 3)
  (Utilities|FlowControl|Sequence
    (:then_0
      ;[0.272 0.8177 0.1576] Material Reactivity
      (Rendering|Material|SetScalarParameterValueonMaterials _returnvalue_3 "Hide Element 01" 0.0)
      (Rendering|Material|SetScalarParameterValueonMaterials _returnvalue_3 "Hide Element 02" 0.0)
      (Rendering|Material|SetScalarParameterValueonMaterials _returnvalue_3 "SP_Current_Stage" 2.0)
      (CallFunction|Stage2FX)
      (Rendering|Material|SetScalarParameterValueonMaterials (CustomCharacter|GetSkeletalMeshForPartOnPlayerPawn) "SP_CurrentTime_IceShards" (Utilities|Time|GetGameTimeinSeconds))
      (Utilities|FlowControl|Delay 1.5)
      (if (Utilities|Enum|Equal(Enum) (CustomCharacter|GetCharacterPartType) "Body")
        (Rendering|Material|SetScalarParameterValueonMaterials _returnvalue_3 "Hide Element 10" 1.0)))
    (:then_1
      (if (< _current_cycle 1)
        _
        (elif (> _current_cycle 6)
          (Rendering|Material|SetScalarParameterValueonMaterials _returnvalue "SP_CurrentTime_Wave" _returnvalue_2)
          (else
            (Rendering|Material|SetScalarParameterValueonMaterials _returnvalue "SP_Current_Eliminations" (Math|Conversions|ToFloat(Integer) (Variables|Default|GetCurrent%20Cycle)))
            (Rendering|Material|SetScalarParameterValueonMaterials _returnvalue "SP_CurrentTime_Frost" _returnvalue_2)))))))

(event Custom|Snowfall%204%20Emliminations
  (bind _returnvalue (CustomCharacter|GetSkeletalMeshForPartOnPlayerPawn))
  (bind _returnvalue_2 (Utilities|Time|GetGameTimeinSeconds))
  (bind _returnvalue_3 (CustomCharacter|GetSkeletalMeshForPartOnPlayerPawn))
  (bind _current_cycle (Variables|Default|GetCurrent%20Cycle))
  ;[0.007499 0.2122 0.6584] FrontEnd Stage cycling
  (Variables|Default|SetCurrent%20Cycle 4)
  (Utilities|FlowControl|Sequence
    (:then_0
      ;[0.272 0.8177 0.1576] Material Reactivity
      (Rendering|Material|SetScalarParameterValueonMaterials _returnvalue_3 "Hide Element 01" 0.0)
      (Rendering|Material|SetScalarParameterValueonMaterials _returnvalue_3 "Hide Element 02" 0.0)
      (Rendering|Material|SetScalarParameterValueonMaterials _returnvalue_3 "SP_Current_Stage" 2.0)
      (if (Utilities|Enum|Equal(Enum) (CustomCharacter|GetCharacterPartType) "Body")
        (Rendering|Material|SetScalarParameterValueonMaterials _returnvalue_3 "Hide Element 10" 1.0)))
    (:then_1
      (if (< _current_cycle 1)
        _
        (elif (> _current_cycle 6)
          (Rendering|Material|SetScalarParameterValueonMaterials _returnvalue "SP_CurrentTime_Wave" _returnvalue_2)
          (else
            (Rendering|Material|SetScalarParameterValueonMaterials _returnvalue "SP_Current_Eliminations" (Math|Conversions|ToFloat(Integer) (Variables|Default|GetCurrent%20Cycle)))
            (Rendering|Material|SetScalarParameterValueonMaterials _returnvalue "SP_CurrentTime_Frost" _returnvalue_2)))))))

(event Custom|Snowfall%205%20Emliminations
  (bind _returnvalue (CustomCharacter|GetSkeletalMeshForPartOnPlayerPawn))
  (bind _returnvalue_2 (Utilities|Time|GetGameTimeinSeconds))
  (bind _returnvalue_3 (CustomCharacter|GetSkeletalMeshForPartOnPlayerPawn))
  (bind _current_cycle (Variables|Default|GetCurrent%20Cycle))
  ;[0.007499 0.2122 0.6584] FrontEnd Stage cycling
  (Variables|Default|SetCurrent%20Cycle 5)
  (Utilities|FlowControl|Sequence
    (:then_0
      ;[0.272 0.8177 0.1576] Material Reactivity
      (Rendering|Material|SetScalarParameterValueonMaterials _returnvalue_3 "Hide Element 01" 0.0)
      (Rendering|Material|SetScalarParameterValueonMaterials _returnvalue_3 "Hide Element 02" 0.0)
      (Rendering|Material|SetScalarParameterValueonMaterials _returnvalue_3 "SP_Current_Stage" 2.0)
      (if (Utilities|Enum|Equal(Enum) (CustomCharacter|GetCharacterPartType) "Body")
        (Rendering|Material|SetScalarParameterValueonMaterials _returnvalue_3 "Hide Element 10" 1.0)))
    (:then_1
      (if (< _current_cycle 1)
        _
        (elif (> _current_cycle 6)
          (Rendering|Material|SetScalarParameterValueonMaterials _returnvalue "SP_CurrentTime_Wave" _returnvalue_2)
          (else
            (Rendering|Material|SetScalarParameterValueonMaterials _returnvalue "SP_Current_Eliminations" (Math|Conversions|ToFloat(Integer) (Variables|Default|GetCurrent%20Cycle)))
            (Rendering|Material|SetScalarParameterValueonMaterials _returnvalue "SP_CurrentTime_Frost" _returnvalue_2)))))))

(event Custom|Snowfall%206%20Emliminations
  (bind _returnvalue (CustomCharacter|GetSkeletalMeshForPartOnPlayerPawn))
  (bind _returnvalue_2 (Utilities|Time|GetGameTimeinSeconds))
  (bind _returnvalue_3 (CustomCharacter|GetSkeletalMeshForPartOnPlayerPawn))
  (bind _current_cycle (Variables|Default|GetCurrent%20Cycle))
  ;[0.007499 0.2122 0.6584] FrontEnd Stage cycling
  (Variables|Default|SetCurrent%20Cycle 6)
  (Utilities|FlowControl|Sequence
    (:then_0
      ;[0.272 0.8177 0.1576] Material Reactivity
      (Rendering|Material|SetScalarParameterValueonMaterials _returnvalue_3 "Hide Element 01" 0.0)
      (Rendering|Material|SetScalarParameterValueonMaterials _returnvalue_3 "Hide Element 02" 0.0)
      (Rendering|Material|SetScalarParameterValueonMaterials _returnvalue_3 "Hide Element 03" 0.0)
      (Rendering|Material|SetScalarParameterValueonMaterials _returnvalue_3 "SP_Current_Stage" 3.0)
      (CallFunction|Stage3FX)
      (Rendering|Material|SetScalarParameterValueonMaterials (CustomCharacter|GetSkeletalMeshForPartOnPlayerPawn) "SP_CurrentTime_IceShards" (Utilities|Time|GetGameTimeinSeconds)))
    (:then_1
      (if (< _current_cycle 1)
        _
        (elif (> _current_cycle 6)
          (Rendering|Material|SetScalarParameterValueonMaterials _returnvalue "SP_CurrentTime_Wave" _returnvalue_2)
          (else
            (Rendering|Material|SetScalarParameterValueonMaterials _returnvalue "SP_Current_Eliminations" (Math|Conversions|ToFloat(Integer) (Variables|Default|GetCurrent%20Cycle)))
            (Rendering|Material|SetScalarParameterValueonMaterials _returnvalue "SP_CurrentTime_Frost" _returnvalue_2)))))))

(event EventOnStatChanged
  (params (State%20Name :type Name :default "None") (Stat%20Value :type Integer))
  (bind _returnvalue (CustomCharacter|GetSkeletalMeshForPartOnPlayerPawn))
  (bind _returnvalue_2 (CustomCharacter|GetCharacterPartType))
  (bind _returnvalue_3 (CustomCharacter|GetSkeletalMeshForPartOnPlayerPawn))
  (bind _returnvalue_4 (Utilities|Time|GetGameTimeinSeconds))
  (bind _returnvalue_5 (Utilities|Time|GetGameTimeinSeconds))
  (bind _returnvalue_6 (CustomCharacter|GetSkeletalMeshForPartOnPlayerPawn))
  (bind _returnvalue_7 (Utilities|Enum|Equal(Enum) _returnvalue_2 "Body"))
  (bind _current_eliminations (Variables|Default|GetCurrent%20Eliminations))
  ;[0.272 0.8177 0.1576] Material Reactivity
  (|CallParentFunction State%20Name Stat%20Value)
  (Variables|Default|SetCurrent%20Eliminations Stat%20Value)
  (Utilities|FlowControl|Sequence
    (:then_0
      (switch int (Variables|Default|GetCurrent%20Eliminations)
        (:0
          (Rendering|Material|SetScalarParameterValueonMaterials _returnvalue_3 "Hide Element 01" 1.0)
          (Rendering|Material|SetScalarParameterValueonMaterials _returnvalue_3 "Hide Element 02" 1.0)
          (Rendering|Material|SetScalarParameterValueonMaterials _returnvalue_3 "Hide Element 03" 1.0)
          (Rendering|Material|SetScalarParameterValueonMaterials _returnvalue_3 "Hide Element 10" 0.0)
          (Rendering|Material|SetScalarParameterValueonMaterials _returnvalue_3 "SP_Current_Stage" 0.0)
          (Rendering|Material|SetScalarParameterValueonMaterials _returnvalue_3 "SP_Current_Eliminations" 0.0))
        (:1
          (Rendering|Material|SetScalarParameterValueonMaterials _returnvalue_3 "Hide Element 01" 0.0)
          (Rendering|Material|SetScalarParameterValueonMaterials _returnvalue_3 "SP_Current_Stage" 1.0)
          (CallFunction|Stage1FX))
        (:2
          (Rendering|Material|SetScalarParameterValueonMaterials _returnvalue_3 "Hide Element 01" 0.0)
          (Rendering|Material|SetScalarParameterValueonMaterials _returnvalue_3 "SP_Current_Stage" 1.0))
        (:3
          (Rendering|Material|SetScalarParameterValueonMaterials _returnvalue_3 "Hide Element 01" 0.0)
          (Rendering|Material|SetScalarParameterValueonMaterials _returnvalue_3 "Hide Element 02" 0.0)
          (Rendering|Material|SetScalarParameterValueonMaterials _returnvalue_3 "SP_Current_Stage" 2.0)
          (CallFunction|Stage2FX)
          (Rendering|Material|SetScalarParameterValueonMaterials _returnvalue "SP_CurrentTime_IceShards" _returnvalue_4)
          (Utilities|FlowControl|Delay 1.5)
          (if _returnvalue_7
            (Rendering|Material|SetScalarParameterValueonMaterials _returnvalue_3 "Hide Element 10" 1.0)))
        (:4
          (Rendering|Material|SetScalarParameterValueonMaterials _returnvalue_3 "Hide Element 01" 0.0)
          (Rendering|Material|SetScalarParameterValueonMaterials _returnvalue_3 "Hide Element 02" 0.0)
          (Rendering|Material|SetScalarParameterValueonMaterials _returnvalue_3 "SP_Current_Stage" 2.0)
          (if _returnvalue_7
            (Rendering|Material|SetScalarParameterValueonMaterials _returnvalue_3 "Hide Element 10" 1.0)))
        (:5
          _)
        (:6
          (Rendering|Material|SetScalarParameterValueonMaterials _returnvalue_3 "Hide Element 01" 0.0)
          (Rendering|Material|SetScalarParameterValueonMaterials _returnvalue_3 "Hide Element 02" 0.0)
          (Rendering|Material|SetScalarParameterValueonMaterials _returnvalue_3 "Hide Element 03" 0.0)
          (Rendering|Material|SetScalarParameterValueonMaterials _returnvalue_3 "SP_Current_Stage" 3.0)
          (CallFunction|Stage3FX)))
      (Rendering|Material|SetScalarParameterValueonMaterials _returnvalue "SP_CurrentTime_IceShards" _returnvalue_4))
    (:then_1
      (if (< _current_eliminations 1)
        _
        (elif (> _current_eliminations 6)
          (Rendering|Material|SetScalarParameterValueonMaterials _returnvalue_6 "SP_CurrentTime_Wave" _returnvalue_5)
          (else
            (Rendering|Material|SetScalarParameterValueonMaterials _returnvalue_6 "SP_Current_Eliminations" (Math|Conversions|ToFloat(Integer) (Variables|Default|GetCurrent%20Eliminations)))
            (Rendering|Material|SetScalarParameterValueonMaterials _returnvalue_6 "SP_CurrentTime_Frost" _returnvalue_5)))))))

(event Custom|Stage%201%20FX
  ;[1 0.423 0.2216] FX
  (Niagara|SpawnSystemAttached (Variables|Default|GetStage%201%20niagara) (CustomCharacter|GetSkeletalMeshForPartOnPlayerPawn) "None" "0, 0, 0" "0, 0, 0" "KeepRelativeOffset" true))

(event Custom|Stage%202%20FX
  ;[1 0.423 0.2216] FX
  (Niagara|SpawnSystemAttached (Variables|Default|GetStage%202%20niagara) (CustomCharacter|GetSkeletalMeshForPartOnPlayerPawn) "None" "0, 0, 0" "0, 0, 0" "KeepRelativeOffset" true))

(event Custom|Stage%203%20FX
  ;[1 0.423 0.2216] FX
  (Niagara|SpawnSystemAttached (Variables|Default|GetStage%203%20niagara) (CustomCharacter|GetSkeletalMeshForPartOnPlayerPawn) "None" "0, 0, 0" "0, 0, 0" "KeepRelativeOffset" true))

(fn ConstructionScript)
