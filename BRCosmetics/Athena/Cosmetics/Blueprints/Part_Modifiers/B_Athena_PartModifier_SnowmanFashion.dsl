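; Cosmetic part modifier for SnowmanFashion that drives the idle VFX particle parameters from the owning pawn's jumping and landing.
; Each movement update feeds JumpCounter into JumpTimeRange, then increments the counter while jump force is applied or zeroes it otherwise.
; A DoOnce, with no reset shown, briefly pulses JumpBurst from the movement update.
; Landing briefly pulses LandBurst, using a short delay to switch it back off.
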(event EventOnCharacterMovementUpdated
  (params (DeltaSeconds :type Float%20%28single-precision%29) (OldLocation :type Vector :default "0, 0, 0") (OldVelocity :type Vector :default "0, 0, 0"))
  (bind _returnvalue (CallFunction|GetIdleVFXComponent))
  (|CallParentFunction DeltaSeconds OldLocation OldVelocity)
  (Effects|Components|ParticleSystem|SetFloatParameter (CallFunction|GetIdleVFXComponent) "JumpTimeRange" (Variables|Default|GetJumpCounter))
  (if (Character|IsJumpProvidingForce (CustomCharacter|GetAssociatedPlayerPawn))
    (Math|Float|IncrementFloat (Variables|Default|GetJumpCounter)
      (:%20%20
        _))
    (else
      (Variables|Default|SetJumpCounter 0.0)))
  (Utilities|FlowControl|DoOnce false
    (:Completed
      (Effects|Components|ParticleSystem|SetFloatParameter _returnvalue "JumpBurst" 1.0)
      (Utilities|FlowControl|Delay 0.05)
      (Effects|Components|ParticleSystem|SetFloatParameter _returnvalue "JumpBurst" 0.0))))

(event EventOnPawnLanded
  (bind _returnvalue (CallFunction|GetIdleVFXComponent))
  (|CallParentFunction)
  (Effects|Components|ParticleSystem|SetFloatParameter _returnvalue "LandBurst" 1.0)
  (Utilities|FlowControl|Delay 0.1)
  (Effects|Components|ParticleSystem|SetFloatParameter _returnvalue "LandBurst" 0.0))

(fn ConstructionScript
  (|CallParentFunction))
