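; B_Athena_PartModifier_SnowyCoast is a character part modifier that drives the SnowyCoast outfit's reactive material effect.
; It triggers once for a frontend reactivity preview, and in game when the GainedHealthTimes stat changes, with a ten-second IsHealing cooldown blocking retriggers.
; The trigger requires the ReactiveOn meta tag, then writes current game time into the player mesh materials' GameTimeFromBP parameter, so the shader times the effect.
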(event EventOnPostResIn
  (|CallParentFunction)
  (if (Reactive|Frontend|ShouldDisplayFrontendReactivity 0 "Body")
    (CallFunction|OnSnowyCoastReactivityTriggered)
    (Utilities|FlowControl|Delay 13.0))
  ;[0.3629 0.1311 1] Watch Attributs
  (Cosmetics|WatchCosmeticStat (CustomCharacter|GetAssociatedPlayerPawn) "(TagName=\"StatManager.GainedHealthTimes\")" 0 false
    (:OnStatChanged
      (if (not (Variables|Default|GetIsHealing))
        (CallFunction|OnSnowyCoastReactivityTriggered)
        (Variables|Default|SetIsHealing true)
        (Utilities|FlowControl|Delay 1e+01)
        (Variables|Default|SetIsHealing false)))))

(event Custom|OnSnowyCoastReactivityTriggered
  ; SnowCoastReactivity
  (if (GameplayTags|HasTag (Cosmetics|GatherMetaTags (CustomCharacter|GetAssociatedPlayerPawn) "/BRCosmetics/Athena/Items/Cosmetics/Characters/Character_SnowyCoast.Character_SnowyCoast" :!class /Script/FortniteGame.FortPlayerPawn) "(TagName=\"Cosmetics.MetaTags.ReactiveOn\")")
    (Rendering|Material|SetScalarParameterValueonMaterials (CustomCharacter|GetSkeletalMeshForPartOnPlayerPawn) "GameTimeFromBP" (Utilities|Time|GetGameTimeinSeconds))))

(fn ConstructionScript
  (|CallParentFunction))
