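; Athena cosmetic part modifier Blueprint for the Solid Mist body character part.
; On the post-res-in event it runs parent logic, fetches the idle VFX component and, if valid, sets its translucency sort distance offset.
; The construction script only calls its parent, and the tick event is empty.
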(event EventTick
  (params (DeltaSeconds :type Float%20%28single-precision%29)))

(event EventOnPostResIn
  (bind _returnvalue (CallFunction|GetIdleVFXComponent))
  (|CallParentFunction)
  (Utilities|IsValid _returnvalue
    (:Is%20Valid
      (Rendering|SetTranslucencySortDistanceOffset _returnvalue 5e+01))))

(fn ConstructionScript
  (|CallParentFunction))
