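; Dev part modifier for the SoloSnooze cosmetic that controls how a loading screen texture shows on the character's body material.
; On res-in it caches the body's dynamic material and, in the front end, saves and restores pawn rotation.
; FilterLoadingScreen looks up the texture's name in a data table to disable it or set user-offset and mask-specific UV transform parameters.
; A debug branch runs a duplicate filter with logging.
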(event EventOnDayPhaseChanged
  (params (CurrentDayPhase :type EFortDayPhase%20Enum :default "Morning") (PreviousDayPhase :type EFortDayPhase%20Enum :default "Morning") (bAtCreation :type Boolean)))

(event EventOnNewPlayerWeaponType
  (params (NewType :type EFortItemType%20Enum :default "WorldItem") (OldType :type EFortItemType%20Enum :default "WorldItem")))

(event EventOnWeaponFired)

(event EventOnNewWeapons
  (params (WeaponsAsItems :type Array%20of%20Fort%20Item%20Object%20References)))

(event EventOnPostResIn
  (bind _playerpawn (Variables|Default|GetPlayerPawn))
  (bind (_returnvalue_location _returnvalue_rotation) (Math|Transform|BreakTransform (Transformation|GetActorTransform _playerpawn)))
  (bind _returnvalue_location_x (Math|Vector|BreakVector _returnvalue_location))
  (bind (_returnvalue_location_2 _returnvalue_rotation_2) (Math|Transform|BreakTransform (Transformation|GetActorTransform (Variables|Default|GetPlayerPawn))))
  (bind _returnvalue_location_3 (Math|Transform|BreakTransform (Transformation|GetActorTransform (Variables|Default|GetPlayerPawn))))
  (bind _returnvalue_location_x_2 (Math|Vector|BreakVector _returnvalue_location_3))
  (bind (_returnvalue_location_4 _returnvalue_rotation_3) (Math|Transform|BreakTransform (Transformation|GetActorTransform (Variables|Default|GetPlayerPawn))))
  (bind (_returnvalue_location_5 _returnvalue_rotation_4) (Math|Transform|BreakTransform (Transformation|GetActorTransform (Variables|Default|GetPlayerPawn))))
  (bind (_returnvalue_location_6 _returnvalue_rotation_5) (Math|Transform|BreakTransform (Transformation|GetActorTransform (Variables|Default|GetPlayerPawn))))
  (bind (_returnvalue_location_7 _returnvalue_rotation_6) (Math|Transform|BreakTransform (Transformation|GetActorTransform (Variables|Default|GetPlayerPawn))))
  (|CallParentFunction)
  (Variables|Default|SetPlayerPawn (CustomCharacter|GetAssociatedPlayerPawn))
  (if (Variables|Default|GetEnableDebugBranch)
    (CallFunction|DebugPrint (Variables|Default|GetEnableRotationDebug) (Utilities|String|Append "!!!! RES IN !!!!   | pawn  :  " (Utilities|GetDisplayName (Variables|Default|GetPlayerPawn))))
    (if (Variables|Default|GetIsFirstRun)
      (Variables|Default|SetBodyMesh (Customization|GetSkeletalMeshforPartType (Variables|Default|GetPlayerPawn) "Body" :!class /Script/FortniteGame.FortPlayerPawn))
      (bind _asmaterial_instance_dynamic (Utilities|Casting|CastToMaterialInstanceDynamic (Rendering|Material|GetMaterial (Variables|Default|GetBodyMesh) 0))
        (:then
          (Variables|Default|SetBodyMaterial _asmaterial_instance_dynamic)
          (Variables|Default|SetIsFirstRun false))
        (:CastFailed
          ;[1 0 0 0.5 1]
          (CallFunction|DebugPrint (Variables|Default|GetEnableLSDebug) "Can't find dynamic material
"))))
    (if (CustomCosmeticModifier|IsinFrontEnd)
      (if (> _returnvalue_location_x_2 5e+03)
        (CallFunction|DebugPrint (Variables|Default|GetEnableRotationDebug) (Utilities|String|Append "    => is display location  :  " (Utilities|String|ToString(Rotator) _returnvalue_rotation_3)))
        (if (Utilities|Operators|NotEqual(!=) _returnvalue_rotation_4 (Variables|Default|GetActorRotation) 0.1)
          (CallFunction|DebugPrint (Variables|Default|GetEnableRotationDebug) (Utilities|String|Append "    => LOC  -  RESTORE stored rotation  :  " (Utilities|String|ToString(Rotator) (Variables|Default|GetActorRotation))))
          (Transformation|SetActorRotation (Variables|Default|GetPlayerPawn) (Variables|Default|GetActorRotation) true)
          (Variables|Default|SetBaseRotation (Variables|Default|GetActorRotation))
          (else
            (CallFunction|DebugPrint (Variables|Default|GetEnableRotationDebug) (Utilities|String|Append "    => LOC  -  no change in rotation, SKIP RESTORE  :  " (Utilities|String|ToString(Rotator) _returnvalue_rotation_6)))))
        (bind _output_get (Variables|Default|SetCurrentRotation _returnvalue_rotation_5))
        (if (Utilities|Operators|NotEqual(!=) _output_get (Variables|Default|GetBaseRotation) 0.1)
          (bind _output_get_2 (Variables|Default|SetActorRotation (Variables|Default|GetCurrentRotation)))
          (CallFunction|DebugPrint (Variables|Default|GetEnableRotationDebug) (Utilities|String|Append "        LOC  -  rotation NOT default, storing  :  " (Utilities|String|ToString(Rotator) _output_get_2)))
          (CallFunction|FilterLoadingScreenDEBUG)
          (Variables|Default|SetEnableLSDebug false)
          (Utilities|FlowControl|RetriggerableDelay (Variables|Default|GetLoopDelay)))
        (CallFunction|DebugPrint (Variables|Default|GetEnableRotationDebug) "        LOC  -  rotation IS default")
        (else
          (CallFunction|DebugPrint (Variables|Default|GetEnableRotationDebug) "    => !! NOT display location  :  ")))
      (else
        (CallFunction|FilterLoadingScreen)))
    (elif (not (Variables|Default|GetIsFirstRun))
      (if (CustomCosmeticModifier|IsinFrontEnd)
        (if (> _returnvalue_location_x 5e+03)
          (if (Utilities|Operators|NotEqual(!=) (Variables|Default|GetActorRotation) _returnvalue_rotation 0.1)
            (Transformation|SetActorRotation (Variables|Default|GetPlayerPawn) (Variables|Default|GetActorRotation) true)
            (Variables|Default|SetBaseRotation (Variables|Default|GetActorRotation)))
          (bind _output_get_3 (Variables|Default|SetCurrentRotation _returnvalue_rotation_2))
          (if (Utilities|Operators|NotEqual(!=) _output_get_3 (Variables|Default|GetBaseRotation) 0.1)
            (Variables|Default|SetActorRotation (Variables|Default|GetCurrentRotation)))
          (CallFunction|FilterLoadingScreen)
          (Variables|Default|SetEnableLSDebug false)
          (Utilities|FlowControl|RetriggerableDelay (Variables|Default|GetLoopDelay)))
        (else
          (CallFunction|FilterLoadingScreen)))
      (else
        (Variables|Default|SetBodyMesh (Customization|GetSkeletalMeshforPartType (Variables|Default|GetPlayerPawn) "Body" :!class /Script/FortniteGame.FortPlayerPawn))
        (bind _asmaterial_instance_dynamic_2 (Utilities|Casting|CastToMaterialInstanceDynamic (Rendering|Material|GetMaterial (Variables|Default|GetBodyMesh) 0)))
        (Variables|Default|SetBodyMaterial _asmaterial_instance_dynamic_2)
        (Variables|Default|SetIsFirstRun false)))))

(event EventOnCurieStateAttached
  (params (Tag :type Gameplay%20Tag%20Structure%20%28by%20ref%29)))

(event EventOnCurieStateDetached
  (params (Tag :type Gameplay%20Tag%20Structure%20%28by%20ref%29)))

(event EventOnBeginSkydiving
  (|CallParentFunction) ; This node is disabled and will not be called.\nDrag off pins to build functionality.
)

(event EventOnEndSkydiving
  (|CallParentFunction) ; This node is disabled and will not be called.\nDrag off pins to build functionality.
)

(event EventOnEnterVehicle
  (params (VehicleEntered :type Fort%20Vehicle%20Interface%20Interface%20%28by%20ref%29))
  (|CallParentFunction VehicleEntered) ; This node is disabled and will not be called.\nDrag off pins to build functionality.
)

(event EventOnExitVehicle
  (|CallParentFunction) ; This node is disabled and will not be called.\nDrag off pins to build functionality.
)

(event EventOnTargetingOrFirstPersonCameraChanged)

(event EventOnHitPawn
  (params (SelfActor :type Actor%20Object%20Reference) (OtherActor :type Actor%20Object%20Reference) (NormalImpulse :type Vector :default "0, 0, 0") (Hit :type Hit%20Result%20Structure%20%28by%20ref%29)))

(fn DebugPrint
  (params (Enable :type Boolean) (String :type String))
  (if Enable
    (CallFunction|DEBUGPrintString (Variables|GetString)))
  (return))

(fn FilterLoadingScreen
  (bind (_name_18_3eccb91848ebbd5d351053a356cf3a2a _allow_3_60034db34313ee00975349a5d4f06337 _allowuserfrontoffset_84_d140d1ff43b7f787274540b2a042d481 _allowuserbackoffset_85_5f4565454c0542e8f7f899970b68b9fb _allowusersleevesoffset_87_da39caff42a98fecc302f8ade2a2667d _overridebasetransforms_82_91aea5ee45b6c7bbd87ccb9b08f4029f _basetransforms_maskfx3_75_40fe718544d16ffc9047b89ec2cb4f99 _basetransforms_maskfx4_76_f6f495a846b73e2e9b2f218d3d70d9ab _basetransforms_maskfx5_77_463cc34643197cc0578ec2b36662b20e _basetransforms_maskfx6_78_4a6b87064869fe79f9a3e9bf1df9607d _basetransforms_maskfx7_79_b3057fa64d1f4ffd8071ff9ef4fc4938 _basetransforms_maskfx8_80_e7261c5a40cf1b856bb21085e1470168 _debugcolor_22_c25a883e4118c3f3b4f24db80139ea46) (Utilities|Struct|BreakCosmeticLoadingScreenRow (LocalVariables|GetThisLoadingScreenParams)))
  (bind (_name_18_3eccb91848ebbd5d351053a356cf3a2a_2 _allow_3_60034db34313ee00975349a5d4f06337_2 _allowuserfrontoffset_84_d140d1ff43b7f787274540b2a042d481_2 _allowuserbackoffset_85_5f4565454c0542e8f7f899970b68b9fb_2 _allowusersleevesoffset_87_da39caff42a98fecc302f8ade2a2667d_2 _overridebasetransforms_82_91aea5ee45b6c7bbd87ccb9b08f4029f_2) (Utilities|Struct|BreakCosmeticLoadingScreenRow (LocalVariables|GetThisLoadingScreenParams)))
  (bind (_name_18_3eccb91848ebbd5d351053a356cf3a2a_3 _allow_3_60034db34313ee00975349a5d4f06337_3) (Utilities|Struct|BreakCosmeticLoadingScreenRow (LocalVariables|GetThisLoadingScreenParams)))
  (bind (_name_18_3eccb91848ebbd5d351053a356cf3a2a_4 _allow_3_60034db34313ee00975349a5d4f06337_4 _allowuserfrontoffset_84_d140d1ff43b7f787274540b2a042d481_3 _allowuserbackoffset_85_5f4565454c0542e8f7f899970b68b9fb_3 _allowusersleevesoffset_87_da39caff42a98fecc302f8ade2a2667d_3 _overridebasetransforms_82_91aea5ee45b6c7bbd87ccb9b08f4029f_3 _basetransforms_maskfx3_75_40fe718544d16ffc9047b89ec2cb4f99_2 _basetransforms_maskfx4_76_f6f495a846b73e2e9b2f218d3d70d9ab_2 _basetransforms_maskfx5_77_463cc34643197cc0578ec2b36662b20e_2 _basetransforms_maskfx6_78_4a6b87064869fe79f9a3e9bf1df9607d_2 _basetransforms_maskfx7_79_b3057fa64d1f4ffd8071ff9ef4fc4938_2 _basetransforms_maskfx8_80_e7261c5a40cf1b856bb21085e1470168_2) (Utilities|Struct|BreakCosmeticLoadingScreenRow (LocalVariables|GetThisBaseTransformsParams)))
  (bind _thisloadingscreenparams (LocalVariables|GetThisLoadingScreenParams))
  (bind (_name_18_3eccb91848ebbd5d351053a356cf3a2a_5 _allow_3_60034db34313ee00975349a5d4f06337_5 _allowuserfrontoffset_84_d140d1ff43b7f787274540b2a042d481_4 _allowuserbackoffset_85_5f4565454c0542e8f7f899970b68b9fb_4 _allowusersleevesoffset_87_da39caff42a98fecc302f8ade2a2667d_4) (Utilities|Struct|BreakCosmeticLoadingScreenRow _thisloadingscreenparams))
  (Utilities|FlowControl|Sequence
    (:then_0
      (bind _returnvalue (Rendering|Material|GetTextureParameterValue (Variables|Default|GetBodyMaterial) "LoadingScreen" :!class /Script/Engine.MaterialInstanceDynamic))
      (LocalVariables|SetThisLoadingScreenAsset _returnvalue)
      (LocalVariables|SetThisLoadingScreenName "")
      (LocalVariables|SetThisFoundLoadingScreenParams false)
      (bind _returnvalue_2 (Utilities|GetDataTableRow (Variables|Default|GetDTLoadingScreens) (Utilities|String|StringToName (Utilities|GetObjectName (LocalVariables|GetThisLoadingScreenAsset)))))
      (bind _output_get (LocalVariables|SetThisLoadingScreenParams _returnvalue_2))
      (LocalVariables|SetThisLoadingScreenName (Utilities|String|ToString(Name) (Utilities|Struct|BreakCosmeticLoadingScreenRow _output_get)))
      (LocalVariables|SetThisFoundLoadingScreenParams true))
    (:then_1
      (if (and (LocalVariables|GetThisFoundLoadingScreenParams) (not _allow_3_60034db34313ee00975349a5d4f06337_3))
        (Rendering|Material|SetVectorParameterValue (Variables|Default|GetBodyMaterial) "Debug Color" (Math|Color|MakeColor 1.0) :!class /Script/Engine.MaterialInstanceDynamic)
        (Rendering|Material|SetScalarParameterValue (Variables|Default|GetBodyMaterial) "LoadingScreenEnabled" 0.0 :!class /Script/Engine.MaterialInstanceDynamic)
        (return)))
    (:then_2
      (if (LocalVariables|GetThisFoundLoadingScreenParams)
        (Rendering|Material|SetVectorParameterValue (Variables|Default|GetBodyMaterial) "Debug Color" _debugcolor_22_c25a883e4118c3f3b4f24db80139ea46 :!class /Script/Engine.MaterialInstanceDynamic)
        (if _overridebasetransforms_82_91aea5ee45b6c7bbd87ccb9b08f4029f_2
          (bind _returnvalue_3 (Rendering|Material|GetScalarParameterValue (Variables|Default|GetBodyMaterial) "Enable Base Transforms" :!class /Script/Engine.MaterialInstanceDynamic))
          (if (> _returnvalue_3 0.99)
            (LocalVariables|SetThisBaseTransformsParams (LocalVariables|GetThisLoadingScreenParams))
            (LocalVariables|SetThisPrintMsg (Utilities|String|Append (LocalVariables|GetThisLoadingScreenName) " : found and applied transform overrides
"))
            (else
              (bind _returnvalue_4 (Utilities|GetDataTableRow (Variables|Default|GetDTLoadingScreens) "Default"))
              (LocalVariables|SetThisBaseTransformsParams _returnvalue_4)
              (LocalVariables|SetThisPrintMsg (Utilities|String|Append (Utilities|GetObjectName (Variables|Default|GetBodyMaterial)) " : transform overrides disabled
"))))
          (bind _returnvalue_5 (Rendering|Material|GetTextureParameterValue (Variables|Default|GetBodyMaterial) "MaskTexture" :!class /Script/Engine.MaterialInstanceDynamic))
          (bind _output_get_2 (LocalVariables|SetThisMaskAsset _returnvalue_5))
          (bind (_lefts _rights _returnvalue_6) (Utilities|String|Split (Utilities|GetObjectName _output_get_2) "_" "IgnoreCase" "FromEnd"))
          (bind (_value _returnvalue_7) (Utilities|Map|Find (Variables|Default|GetMasksMap) _rights))
          (LocalVariables|SetThisMaskIndex _value)
          (if (not (Math|Boolean|NANDBoolean _returnvalue_7 _returnvalue_6))
            (CallFunction|OverrideBaseTransforms (Variables|Default|GetBodyMaterial) (Utilities|Select :Index (LocalVariables|GetThisMaskIndex) :Option%200 _basetransforms_maskfx3_75_40fe718544d16ffc9047b89ec2cb4f99_2 :Option%201 _basetransforms_maskfx4_76_f6f495a846b73e2e9b2f218d3d70d9ab_2 :Option%202 _basetransforms_maskfx5_77_463cc34643197cc0578ec2b36662b20e_2 :Option%203 _basetransforms_maskfx6_78_4a6b87064869fe79f9a3e9bf1df9607d_2 :Option%204 _basetransforms_maskfx7_79_b3057fa64d1f4ffd8071ff9ef4fc4938_2 :Option%205 _basetransforms_maskfx8_80_e7261c5a40cf1b856bb21085e1470168_2))))
        (else
          (Rendering|Material|SetVectorParameterValue (Variables|Default|GetBodyMaterial) "Debug Color" (Math|Color|MakeColor 1.0) :!class /Script/Engine.MaterialInstanceDynamic))))
    (:then_3
      (bind _returnvalue_8 (Rendering|Material|GetScalarParameterValue (Variables|Default|GetBodyMaterial) "Enable User Transforms" :!class /Script/Engine.MaterialInstanceDynamic))
      (if (> _returnvalue_8 0.99)
        (if (LocalVariables|GetThisFoundLoadingScreenParams)
          (LocalVariables|SetThisAllowUserFrontOffset _allowuserfrontoffset_84_d140d1ff43b7f787274540b2a042d481_4)
          (LocalVariables|SetThisAllowUserBackOffset _allowuserbackoffset_85_5f4565454c0542e8f7f899970b68b9fb_4)
          (LocalVariables|SetThisAllowUserSleevesOffset _allowusersleevesoffset_87_da39caff42a98fecc302f8ade2a2667d_4)
          (LocalVariables|SetThisPrintMsg (Utilities|String|Append "allow user transforms [ Front : " (Utilities|String|ToString(Boolean) (LocalVariables|GetThisAllowUserFrontOffset)) ", Back : " (Utilities|String|ToString(Boolean) (LocalVariables|GetThisAllowUserBackOffset)) ", Sleeves : " (Utilities|String|ToString(Boolean) (LocalVariables|GetThisAllowUserSleevesOffset)) " ]"))
          (else
            (LocalVariables|SetThisAllowUserFrontOffset true)
            (LocalVariables|SetThisAllowUserBackOffset true)
            (LocalVariables|SetThisAllowUserSleevesOffset true)
            (LocalVariables|SetThisPrintMsg "all user transforms allowed")))
        (Rendering|Material|SetScalarParameterValue (Variables|Default|GetBodyMaterial) "LoadingScreen_UVTransform_EnableUserFrontOffset" (Math|Conversions|ToFloat(Boolean) (LocalVariables|GetThisAllowUserFrontOffset)) :!class /Script/Engine.MaterialInstanceDynamic)
        (Rendering|Material|SetScalarParameterValue (Variables|Default|GetBodyMaterial) "LoadingScreen_UVTransform_EnableUserBackOffset" (Math|Conversions|ToFloat(Boolean) (LocalVariables|GetThisAllowUserBackOffset)) :!class /Script/Engine.MaterialInstanceDynamic)
        (Rendering|Material|SetScalarParameterValue (Variables|Default|GetBodyMaterial) "LoadingScreen_UVTransform_EnableUserSleevesOffset" (Math|Conversions|ToFloat(Boolean) (LocalVariables|GetThisAllowUserSleevesOffset)) :!class /Script/Engine.MaterialInstanceDynamic)
        (else
          (return))))))

(fn FilterLoadingScreenDEBUG
  (bind (_name_18_3eccb91848ebbd5d351053a356cf3a2a _allow_3_60034db34313ee00975349a5d4f06337 _allowuserfrontoffset_84_d140d1ff43b7f787274540b2a042d481 _allowuserbackoffset_85_5f4565454c0542e8f7f899970b68b9fb _allowusersleevesoffset_87_da39caff42a98fecc302f8ade2a2667d _overridebasetransforms_82_91aea5ee45b6c7bbd87ccb9b08f4029f _basetransforms_maskfx3_75_40fe718544d16ffc9047b89ec2cb4f99 _basetransforms_maskfx4_76_f6f495a846b73e2e9b2f218d3d70d9ab _basetransforms_maskfx5_77_463cc34643197cc0578ec2b36662b20e _basetransforms_maskfx6_78_4a6b87064869fe79f9a3e9bf1df9607d _basetransforms_maskfx7_79_b3057fa64d1f4ffd8071ff9ef4fc4938 _basetransforms_maskfx8_80_e7261c5a40cf1b856bb21085e1470168 _debugcolor_22_c25a883e4118c3f3b4f24db80139ea46) (Utilities|Struct|BreakCosmeticLoadingScreenRow (LocalVariables|GetThisLoadingScreenParams)))
  (bind (_name_18_3eccb91848ebbd5d351053a356cf3a2a_2 _allow_3_60034db34313ee00975349a5d4f06337_2 _allowuserfrontoffset_84_d140d1ff43b7f787274540b2a042d481_2 _allowuserbackoffset_85_5f4565454c0542e8f7f899970b68b9fb_2 _allowusersleevesoffset_87_da39caff42a98fecc302f8ade2a2667d_2 _overridebasetransforms_82_91aea5ee45b6c7bbd87ccb9b08f4029f_2) (Utilities|Struct|BreakCosmeticLoadingScreenRow (LocalVariables|GetThisLoadingScreenParams)))
  (bind (_name_18_3eccb91848ebbd5d351053a356cf3a2a_3 _allow_3_60034db34313ee00975349a5d4f06337_3) (Utilities|Struct|BreakCosmeticLoadingScreenRow (LocalVariables|GetThisLoadingScreenParams)))
  (bind (_name_18_3eccb91848ebbd5d351053a356cf3a2a_4 _allow_3_60034db34313ee00975349a5d4f06337_4 _allowuserfrontoffset_84_d140d1ff43b7f787274540b2a042d481_3 _allowuserbackoffset_85_5f4565454c0542e8f7f899970b68b9fb_3 _allowusersleevesoffset_87_da39caff42a98fecc302f8ade2a2667d_3 _overridebasetransforms_82_91aea5ee45b6c7bbd87ccb9b08f4029f_3 _basetransforms_maskfx3_75_40fe718544d16ffc9047b89ec2cb4f99_2 _basetransforms_maskfx4_76_f6f495a846b73e2e9b2f218d3d70d9ab_2 _basetransforms_maskfx5_77_463cc34643197cc0578ec2b36662b20e_2 _basetransforms_maskfx6_78_4a6b87064869fe79f9a3e9bf1df9607d_2 _basetransforms_maskfx7_79_b3057fa64d1f4ffd8071ff9ef4fc4938_2 _basetransforms_maskfx8_80_e7261c5a40cf1b856bb21085e1470168_2) (Utilities|Struct|BreakCosmeticLoadingScreenRow (LocalVariables|GetThisBaseTransformsParams)))
  (bind _thisloadingscreenparams (LocalVariables|GetThisLoadingScreenParams))
  (bind (_name_18_3eccb91848ebbd5d351053a356cf3a2a_5 _allow_3_60034db34313ee00975349a5d4f06337_5 _allowuserfrontoffset_84_d140d1ff43b7f787274540b2a042d481_4 _allowuserbackoffset_85_5f4565454c0542e8f7f899970b68b9fb_4 _allowusersleevesoffset_87_da39caff42a98fecc302f8ade2a2667d_4) (Utilities|Struct|BreakCosmeticLoadingScreenRow _thisloadingscreenparams))
  (Utilities|FlowControl|Sequence
    (:then_0
      (bind _returnvalue (Rendering|Material|GetTextureParameterValue (Variables|Default|GetBodyMaterial) "LoadingScreen" :!class /Script/Engine.MaterialInstanceDynamic))
      (LocalVariables|SetThisLoadingScreenAsset _returnvalue)
      (LocalVariables|SetThisLoadingScreenName "")
      (LocalVariables|SetThisFoundLoadingScreenParams false)
      (bind _returnvalue_2 (Utilities|GetDataTableRow (Variables|Default|GetDTLoadingScreens) (Utilities|String|StringToName (Utilities|GetObjectName (LocalVariables|GetThisLoadingScreenAsset)))))
      (bind _output_get (LocalVariables|SetThisLoadingScreenParams _returnvalue_2))
      (LocalVariables|SetThisLoadingScreenName (Utilities|String|ToString(Name) (Utilities|Struct|BreakCosmeticLoadingScreenRow _output_get)))
      (LocalVariables|SetThisFoundLoadingScreenParams true))
    (:then_1
      (if (and (LocalVariables|GetThisFoundLoadingScreenParams) (not _allow_3_60034db34313ee00975349a5d4f06337_3))
        (Rendering|Material|SetVectorParameterValue (Variables|Default|GetBodyMaterial) "Debug Color" (Math|Color|MakeColor 1.0) :!class /Script/Engine.MaterialInstanceDynamic)
        (Rendering|Material|SetScalarParameterValue (Variables|Default|GetBodyMaterial) "LoadingScreenEnabled" 0.0 :!class /Script/Engine.MaterialInstanceDynamic)
        ;[1 0 0 0.5 1]
        (CallFunction|DebugPrint (Variables|Default|GetEnableLSDebug) (Utilities|String|Append (Utilities|GetObjectName (LocalVariables|GetThisLoadingScreenAsset)) " : Loading Screen not allowed" "
"))
        (return)
        (else
          ;[1 0 0 0.5 1]
          (CallFunction|DebugPrint (Variables|Default|GetEnableLSDebug) (Utilities|String|Append "Assigned Loading Screen : " (Utilities|GetObjectName (LocalVariables|GetThisLoadingScreenAsset)) "
")))))
    (:then_2
      (if (LocalVariables|GetThisFoundLoadingScreenParams)
        (Rendering|Material|SetVectorParameterValue (Variables|Default|GetBodyMaterial) "Debug Color" _debugcolor_22_c25a883e4118c3f3b4f24db80139ea46 :!class /Script/Engine.MaterialInstanceDynamic)
        (if _overridebasetransforms_82_91aea5ee45b6c7bbd87ccb9b08f4029f_2
          (bind _returnvalue_3 (Rendering|Material|GetScalarParameterValue (Variables|Default|GetBodyMaterial) "Enable Base Transforms" :!class /Script/Engine.MaterialInstanceDynamic))
          (if (> _returnvalue_3 0.99)
            (LocalVariables|SetThisBaseTransformsParams (LocalVariables|GetThisLoadingScreenParams))
            (LocalVariables|SetThisPrintMsg (Utilities|String|Append (LocalVariables|GetThisLoadingScreenName) " : found and applied transform overrides
"))
            (else
              (bind _returnvalue_4 (Utilities|GetDataTableRow (Variables|Default|GetDTLoadingScreens) "Default"))
              (LocalVariables|SetThisBaseTransformsParams _returnvalue_4)
              (LocalVariables|SetThisPrintMsg (Utilities|String|Append (Utilities|GetObjectName (Variables|Default|GetBodyMaterial)) " : transform overrides disabled
"))))
          (bind _returnvalue_5 (Rendering|Material|GetTextureParameterValue (Variables|Default|GetBodyMaterial) "MaskTexture" :!class /Script/Engine.MaterialInstanceDynamic))
          (bind _output_get_2 (LocalVariables|SetThisMaskAsset _returnvalue_5))
          (bind (_lefts _rights _returnvalue_6) (Utilities|String|Split (Utilities|GetObjectName _output_get_2) "_" "IgnoreCase" "FromEnd"))
          (bind (_value _returnvalue_7) (Utilities|Map|Find (Variables|Default|GetMasksMap) _rights))
          (LocalVariables|SetThisMaskIndex _value)
          (if (Math|Boolean|NANDBoolean _returnvalue_7 _returnvalue_6)
            ;[1 0 0 0.5 1]
            (CallFunction|DebugPrint (Variables|Default|GetEnableLSDebug) (Utilities|String|Append (Utilities|GetObjectName (LocalVariables|GetThisMaskAsset)) " : invalid mask texture
"))
            (else
              (CallFunction|OverrideBaseTransforms (Variables|Default|GetBodyMaterial) (Utilities|Select :Index (LocalVariables|GetThisMaskIndex) :Option%200 _basetransforms_maskfx3_75_40fe718544d16ffc9047b89ec2cb4f99_2 :Option%201 _basetransforms_maskfx4_76_f6f495a846b73e2e9b2f218d3d70d9ab_2 :Option%202 _basetransforms_maskfx5_77_463cc34643197cc0578ec2b36662b20e_2 :Option%203 _basetransforms_maskfx6_78_4a6b87064869fe79f9a3e9bf1df9607d_2 :Option%204 _basetransforms_maskfx7_79_b3057fa64d1f4ffd8071ff9ef4fc4938_2 :Option%205 _basetransforms_maskfx8_80_e7261c5a40cf1b856bb21085e1470168_2))
              ;[1 0 0 0.5 1]
              (CallFunction|DebugPrint (Variables|Default|GetEnableLSDebug) (LocalVariables|GetThisPrintMsg))))
          (else
            ;[1 0 0 0.5 1]
            (CallFunction|DebugPrint (Variables|Default|GetEnableLSDebug) (Utilities|String|Append (LocalVariables|GetThisLoadingScreenName) " : no transform overrides
"))))
        (else
          (Rendering|Material|SetVectorParameterValue (Variables|Default|GetBodyMaterial) "Debug Color" (Math|Color|MakeColor 1.0) :!class /Script/Engine.MaterialInstanceDynamic)
          ;[1 0 0 0.5 1]
          (CallFunction|DebugPrint (Variables|Default|GetEnableLSDebug) (Utilities|String|Append (Utilities|GetObjectName (LocalVariables|GetThisLoadingScreenAsset)) " : Loading Screen not in Data Table, " "using default transforms" "
")))))
    (:then_3
      (bind _returnvalue_8 (Rendering|Material|GetScalarParameterValue (Variables|Default|GetBodyMaterial) "Enable User Transforms" :!class /Script/Engine.MaterialInstanceDynamic))
      (if (> _returnvalue_8 0.99)
        (if (LocalVariables|GetThisFoundLoadingScreenParams)
          (LocalVariables|SetThisAllowUserFrontOffset _allowuserfrontoffset_84_d140d1ff43b7f787274540b2a042d481_4)
          (LocalVariables|SetThisAllowUserBackOffset _allowuserbackoffset_85_5f4565454c0542e8f7f899970b68b9fb_4)
          (LocalVariables|SetThisAllowUserSleevesOffset _allowusersleevesoffset_87_da39caff42a98fecc302f8ade2a2667d_4)
          (LocalVariables|SetThisPrintMsg (Utilities|String|Append "allow user transforms [ Front : " (Utilities|String|ToString(Boolean) (LocalVariables|GetThisAllowUserFrontOffset)) ", Back : " (Utilities|String|ToString(Boolean) (LocalVariables|GetThisAllowUserBackOffset)) ", Sleeves : " (Utilities|String|ToString(Boolean) (LocalVariables|GetThisAllowUserSleevesOffset)) " ]"))
          (else
            (LocalVariables|SetThisAllowUserFrontOffset true)
            (LocalVariables|SetThisAllowUserBackOffset true)
            (LocalVariables|SetThisAllowUserSleevesOffset true)
            (LocalVariables|SetThisPrintMsg "all user transforms allowed")))
        (Rendering|Material|SetScalarParameterValue (Variables|Default|GetBodyMaterial) "LoadingScreen_UVTransform_EnableUserFrontOffset" (Math|Conversions|ToFloat(Boolean) (LocalVariables|GetThisAllowUserFrontOffset)) :!class /Script/Engine.MaterialInstanceDynamic)
        (Rendering|Material|SetScalarParameterValue (Variables|Default|GetBodyMaterial) "LoadingScreen_UVTransform_EnableUserBackOffset" (Math|Conversions|ToFloat(Boolean) (LocalVariables|GetThisAllowUserBackOffset)) :!class /Script/Engine.MaterialInstanceDynamic)
        (Rendering|Material|SetScalarParameterValue (Variables|Default|GetBodyMaterial) "LoadingScreen_UVTransform_EnableUserSleevesOffset" (Math|Conversions|ToFloat(Boolean) (LocalVariables|GetThisAllowUserSleevesOffset)) :!class /Script/Engine.MaterialInstanceDynamic)
        ;[1 0 0 0.5 1]
        (CallFunction|DebugPrint (Variables|Default|GetEnableLSDebug) (Utilities|String|Append (LocalVariables|GetThisLoadingScreenName) " : " (LocalVariables|GetThisPrintMsg) "
"))
        (else
          ;[1 0 0 0.5 1]
          (CallFunction|DebugPrint (Variables|Default|GetEnableLSDebug) (Utilities|String|Append (Utilities|GetObjectName (Variables|Default|GetBodyMaterial)) " : user transforms disabled
"))
          (return))))))

(fn OverrideBaseTransforms
  (params (Material :type Material%20Instance%20Dynamic%20Object%20Reference) (Transforms :type Cosmetic%20Loading%20Screen%20Base%20Transforms%20Structure))
  (bind (_front_16_a05bc07248f73c6be2298d882a74a1b8 _back_17_732a4eef4c4194f07c6947a281a7f419 _sleeves_18_7695d8f24511c88a386973866d328505 _leftsleeveoffset_21_e16835e845a314cbe66562b39a2fcd40) (Utilities|Struct|BreakCosmeticLoadingScreenBaseTransforms Transforms))
  (bind (_offset_17_9d965b8d482136f8bc3b70bddd090112 _rotation_8_08972406401a4f5e80bf39b352b6aedf _scale_10_3897cb6b455f322587bcae8b5c55d5a2) (Utilities|Struct|BreakCosmeticLoadingScreenMaskTransforms (LocalVariables|GetThisTransform)))
  (bind (_offset_17_9d965b8d482136f8bc3b70bddd090112_x _offset_17_9d965b8d482136f8bc3b70bddd090112_y) (Math|Vector2D|BreakVector2D _offset_17_9d965b8d482136f8bc3b70bddd090112))
  (LocalVariables|SetThisBaseTransforms (Utilities|Array|MakeArray _front_16_a05bc07248f73c6be2298d882a74a1b8 _back_17_732a4eef4c4194f07c6947a281a7f419 _sleeves_18_7695d8f24511c88a386973866d328505))
  (LocalVariables|SetThisLeftSleeveOffset _leftsleeveoffset_21_e16835e845a314cbe66562b39a2fcd40)
  (Utilities|FlowControl|Sequence
    (:then_0
      (for (_array_element _array_index) (LocalVariables|GetThisBaseTransforms)
        (LocalVariables|SetThisTransform _array_element)
        (Rendering|Material|SetVectorParameterValue (Variables|GetMaterial) (Utilities|Array|Get(aref) (LocalVariables|GetThisMaterialParams) _array_index) (Math|Color|MakeColor _offset_17_9d965b8d482136f8bc3b70bddd090112_x _offset_17_9d965b8d482136f8bc3b70bddd090112_y _rotation_8_08972406401a4f5e80bf39b352b6aedf _scale_10_3897cb6b455f322587bcae8b5c55d5a2) :!class /Script/Engine.MaterialInstanceDynamic)))
    (:then_1
      (Rendering|Material|SetScalarParameterValue (Variables|GetMaterial) (Utilities|Array|Get(aref) (LocalVariables|GetThisMaterialParams) (- (Utilities|Array|Length (LocalVariables|GetThisMaterialParams)) 1)) (LocalVariables|GetThisLeftSleeveOffset) :!class /Script/Engine.MaterialInstanceDynamic))))

(fn ConstructionScript
  (|CallParentFunction))
