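; Cosmetic part modifier for SpacePlungeHunt that drives a reactive material effect on the character part's skeletal mesh.
; Any stat change stamps the current game time into ReactiveTime, so the material can time its own reaction without per-frame updates.
; After res-in, it flags IsInFrontEnd; in the front end, the FrontReactive timeline animates ReactiveAlpha.
; Outside the front end, it backdates ReactiveTime to a negative value, presumably so the effect starts idle.
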(event EventOnStatChanged
  (params (State%20Name :type Name :default "None") (Stat%20Value :type Integer))
  (|CallParentFunction State%20Name Stat%20Value)
  (Rendering|Material|SetScalarParameterValueonMaterials (CustomCharacter|GetSkeletalMeshforPartonOwner) "ReactiveTime" (Utilities|Time|GetGameTimeinSeconds)))

(event EventOnPostResIn
  (bind _returnvalue (CustomCharacter|GetSkeletalMeshforPartonOwner))
  (|CallParentFunction)
  (if (CustomCosmeticModifier|IsinFrontEnd)
    (Rendering|Material|SetScalarParameterValueonMaterials _returnvalue "IsInFrontEnd" 1.0)
    (bind (_direction _reactivealpha) (Timeline|FrontReactive PlayFromStart)
      (:Update
        (Rendering|Material|SetScalarParameterValueonMaterials _returnvalue "ReactiveAlpha" _reactivealpha)))
    (else
      (Rendering|Material|SetScalarParameterValueonMaterials _returnvalue "ReactiveTime" -2e+01)
      (Rendering|Material|SetScalarParameterValueonMaterials _returnvalue "IsInFrontEnd" 0.0))))

(fn ConstructionScript
  (|CallParentFunction))
