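; Athena cosmetic part modifier Blueprint for Space Wanderer that turns its idle VFX on only at night.
; On a day phase change it calls the parent handler, then activates the idle VFX component if the new phase is Night and deactivates it otherwise.
; The construction script only calls its parent.
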(event EventOnDayPhaseChanged
  (params (CurrentDayPhase :type EFortDayPhase%20Enum :default "Morning") (PreviousDayPhase :type EFortDayPhase%20Enum :default "Morning") (bAtCreation :type Boolean))
  (bind _returnvalue (CallFunction|GetIdleVFXComponent))
  (|CallParentFunction CurrentDayPhase PreviousDayPhase bAtCreation)
  (if (Utilities|Enum|Equal(Enum) CurrentDayPhase "Night")
    (Components|Activation|Activate _returnvalue false)
    (else
      (Components|Activation|Deactivate _returnvalue))))

(fn ConstructionScript
  (|CallParentFunction))
