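; Speedy cosmetic part modifier that changes its idle particle effect based on how fast the character moves.
; On each character movement update, it compares the absolute X and Y of the old velocity separately against a threshold variable.
; If either axis exceeds it, the idle VFX component's SecondaryAlpha switches on and LightScale doubles; otherwise both return to base values, with no blending.
; The post-res-in event and construction script only call their parents.
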(event EventOnPostResIn
  (|CallParentFunction))

(event EventOnCharacterMovementUpdated
  (params (DeltaSeconds :type Float%20%28single-precision%29) (OldLocation :type Vector :default "0, 0, 0") (OldVelocity :type Vector :default "0, 0, 0"))
  (bind _x (.x OldVelocity))
  (bind _y (.y OldVelocity))
  (bind _velocity_check_amount (Variables|Default|GetVelocity%20Check%20Amount))
  (bind _returnvalue (Math|Float|Absolute(Float) _x))
  (bind _returnvalue_2 (> _returnvalue _velocity_check_amount))
  (bind _returnvalue_3 (Math|Float|Absolute(Float) _y))
  (bind _returnvalue_4 (> _returnvalue_3 _velocity_check_amount))
  (bind _returnvalue_5 (CallFunction|GetIdleVFXComponent))
  (bind _returnvalue_6 (or _returnvalue_2 _returnvalue_4))
  (|CallParentFunction DeltaSeconds OldLocation OldVelocity)
  (Effects|Components|ParticleSystem|SetFloatParameter _returnvalue_5 "SecondaryAlpha" (select _returnvalue_6 1.0 0.0))
  (Effects|Components|ParticleSystem|SetFloatParameter _returnvalue_5 "LightScale" (select _returnvalue_6 2.0 1.0)))

(fn ConstructionScript
  (|CallParentFunction))
