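; Cosmetic part modifier for SpireTendFlair that makes an idle particle effect react in stages, apparently to the player's eliminations.
; Each positive stat change increments ElimCount up to a maximum, activates the idle VFX and passes the count as a particle parameter.
; Reaching stage four arms a Tick-driven countdown that resets the effect to zero after a few seconds.
; In the front end, a looping timer drives the effect to preview the stages.
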(event EventOnPostResIn
  (|CallParentFunction)
  (if (CustomCosmeticModifier|IsinFrontEnd)
    (Utilities|Time|SetTimerbyEvent (delegate Custom|FE_Reactivity) 4.0 true)
    (CallFunction|FEReactivity)))

(event Custom|FE_Reactivity
  (bind _returnvalue (CallFunction|GetIdleVFXComponent))
  (Components|Activation|Activate _returnvalue true)
  (Effects|Components|ParticleSystem|SetIntParameter _returnvalue "ElimCount" (Variables|Default|GetElimCount))
  (bind _result (Math|Integer|IncrementInt (Variables|Default|GetElimCount))
    (:%20%20
      (Math|Integer|IncrementInt (- (Math|Integer|%25%28Integer%29 _result 5) 1)))))

(event EventTick
  (params (DeltaSeconds :type Float%20%28single-precision%29))
  (bind _returnvalue (CallFunction|GetIdleVFXComponent))
  (|CallParentFunction 0.0)
  (if (Variables|Default|GetResetTimer)
    (bind _output_get (Variables|Default|SetTimeSinceMaxStage (+ DeltaSeconds (Variables|Default|GetTimeSinceMaxStage))))
    (if (>= _output_get 4.0)
      (bind _output_get_2 (Variables|Default|SetElimCount 0))
      (Variables|Default|SetResetTimer false)
      (Variables|Default|SetTimeSinceMaxStage 0.0)
      (Components|Activation|Activate _returnvalue true)
      (Effects|Components|ParticleSystem|SetIntParameter _returnvalue "ElimCount" _output_get_2))))

(event EventOnStatChanged
  (params (State%20Name :type Name :default "None") (Stat%20Value :type Integer))
  (bind _returnvalue (CallFunction|GetIdleVFXComponent))
  (bind _elimcount (Variables|Default|GetElimCount))
  (|CallParentFunction State%20Name Stat%20Value)
  (if (> Stat%20Value 0)
    (bind _result (Math|Integer|IncrementInt _elimcount)
      (:%20%20
        (bind _output_get (Variables|Default|SetElimCount (Math|Integer|Clamp(Integer) _result _elimcount (Variables|Default|GetMaxElimCount))))
        (switch int _output_get
          (:0
            _)
          (:1
            _)
          (:2
            _)
          (:3
            (Components|Activation|Activate _returnvalue true)
            (Effects|Components|ParticleSystem|SetIntParameter _returnvalue "ElimCount" _output_get))
          (:4
            (Components|Activation|Activate _returnvalue true)
            (Effects|Components|ParticleSystem|SetIntParameter _returnvalue "ElimCount" _output_get)
            (Variables|Default|SetResetTimer true)
            (Variables|Default|SetTimeSinceMaxStage 0.0)))
        (Components|Activation|Activate _returnvalue true)
        (Effects|Components|ParticleSystem|SetIntParameter _returnvalue "ElimCount" _output_get)))))

(fn ConstructionScript
  (|CallParentFunction))
