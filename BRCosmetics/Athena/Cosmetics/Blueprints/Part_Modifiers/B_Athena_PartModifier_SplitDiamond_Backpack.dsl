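; Backpack cosmetic part modifier for Split Diamond that drives a reactive emissive glow on the part's skeletal mesh.
; After res-in, the backpack's Particle variant channel sets whether it is reactive, zeroing emissive for the non-reactive option.
; Day phase changes raise emissive to its maximum at Night when reactive, otherwise switch it off.
; In the front end, the reactive variant plays a timeline driving both emissive and the idle VFX CycleFX parameter.
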(event EventOnDayPhaseChanged
  (params (CurrentDayPhase :type EFortDayPhase%20Enum :default "Morning") (PreviousDayPhase :type EFortDayPhase%20Enum :default "Morning") (bAtCreation :type Boolean))
  (bind _returnvalue (CustomCharacter|GetSkeletalMeshForPartOnPlayerPawn))
  (|CallParentFunction CurrentDayPhase PreviousDayPhase bAtCreation)
  (if (Variables|Default|GetIsReactive)
    (if (Utilities|Enum|Equal(Enum) CurrentDayPhase "Night")
      (Rendering|Material|SetScalarParameterValueonMaterials _returnvalue "emissive mult" (Variables|Default|GetLightEmissiveMultMax))
      (else
        _)))
  (Rendering|Material|SetScalarParameterValueonMaterials _returnvalue "emissive mult" 0.0))

(event EventOnPostResIn
  (|CallParentFunction)
  (bind _asplayer_pawn_athena (Utilities|Casting|CastToPlayerPawn_Athena (CustomCharacter|GetAssociatedPlayerPawn)))
  (bind _index (CosmeticLoadout|GetVariantChannelActiveIndexforSlotActor "(TagName=\"LoadoutSlot.Backpack\")" _asplayer_pawn_athena "(TagName=\"Cosmetics.Variant.Channel.Particle\")" (Utilities|Array|MakeArray "(TagName=\"Cosmetics.Variant.Property.Particle0\")" "(TagName=\"Cosmetics.Variant.Property.Particle1\")") false false))
  (switch int _index
    (:0
      (Variables|Default|SetIsReactive true)
      ;[1 1 1] FE Cycling
      (if (CustomCosmeticModifier|IsinFrontEnd)
        (bind (_direction _lightintensity) (Timeline|FE%20Cycling PlayFromStart)
          (:Update
            (Rendering|Material|SetScalarParameterValueonMaterials (CustomCharacter|GetSkeletalMeshForPartOnPlayerPawn) "emissive mult" (* _lightintensity (Variables|Default|GetLightEmissiveMultMax)))
            (Effects|Components|ParticleSystem|SetFloatParameter (CallFunction|GetIdleVFXComponent) "CycleFX" _lightintensity)))))
    (:1
      (Variables|Default|SetIsReactive false)
      (Rendering|Material|SetScalarParameterValueonMaterials (CustomCharacter|GetSkeletalMeshForPartOnPlayerPawn) "emissive mult" 0.0))))

(fn ConstructionScript
  (|CallParentFunction))
