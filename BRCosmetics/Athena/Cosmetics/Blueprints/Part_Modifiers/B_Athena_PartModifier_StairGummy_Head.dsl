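; Cosmetic part modifier for the StairGummy character head that feeds the head bone's world-space position and orientation to its material.
; Every tick it reads the "head" socket transform from a character part mesh and gets a dynamic material instance for a part mesh.
; It enables Nose_UseWS and writes the bone location plus right and up vectors, remapping bone axes (negated up as right, forward as up).
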(event EventTick
  (params (DeltaSeconds :type Float%20%28single-precision%29))
  (bind _returnvalue (CustomCharacter|GetSkeletalMeshForPartOnPlayerPawn))
  (bind _returnvalue_2 (CustomCharacter|GetSkeletalMeshForPartOnPlayerPawn))
  (bind _returnvalue_3 (Transformation|GetSocketTransform _returnvalue_2 "head" "RTS_World"))
  (bind _rotation (.rotation _returnvalue_3))
  (bind _material_instance_dynamic (Class|BTechArtGlobalLibrary|GetDynamicMaterialInstance _returnvalue))
  (Rendering|Material|SetScalarParameterValue _material_instance_dynamic "Nose_UseWS" 1.0 :!class /Script/Engine.MaterialInstanceDynamic)
  (Rendering|Material|SetVectorParameterValue _material_instance_dynamic "HeadBone_VecRight" (Math|Conversions|ToLinearColor(Vector) (* (Math|Vector|GetUpVector _rotation) -1.0)) :!class /Script/Engine.MaterialInstanceDynamic)
  (Rendering|Material|SetVectorParameterValue _material_instance_dynamic "HeadBone_VecUp" (Math|Conversions|ToLinearColor(Vector) (Math|Vector|GetForwardVector _rotation)) :!class /Script/Engine.MaterialInstanceDynamic)
  (Rendering|Material|SetVectorParameterValue _material_instance_dynamic "HeadBone_Location" (Math|Conversions|ToLinearColor(Vector) (.location _returnvalue_3)) :!class /Script/Engine.MaterialInstanceDynamic))

(fn ConstructionScript
  (|CallParentFunction))
