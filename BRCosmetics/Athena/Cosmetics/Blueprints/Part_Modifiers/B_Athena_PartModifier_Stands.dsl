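; Character part modifier Blueprint for the "Stands" Athena cosmetic that derives a material hue-shift value from its audio colors.
; On post res-in it calls the parent, then reads Audio Primary and Secondary Color from the part mesh's dynamic material into variables.
; It converts both colors to HSV and stores the shortest hue path between their hues.
; That result is written to the "Shift Max" scalar on the mesh's materials.
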(event EventOnPostResIn
  (bind _returnvalue (CustomCharacter|GetSkeletalMeshForPartOnPlayerPawn))
  (bind _mid (Material|GetMID _returnvalue))
  (bind _hsv_x (Math|Vector|BreakVector (Utilities|RGBToHSV (Variables|Default|GetAudio%20Secondary%20Color))))
  (bind _hsv_x_2 (Math|Vector|BreakVector (Utilities|RGBToHSV (Variables|Default|GetAudio%20Primary%20Color))))
  (|CallParentFunction)
  (bind _returnvalue_2 (Rendering|Material|GetVectorParameterValue _mid "Audio Primary Color" :!class /Script/Engine.MaterialInstanceDynamic))
  (Variables|Default|SetAudio%20Primary%20Color _returnvalue_2)
  (bind _returnvalue_3 (Rendering|Material|GetVectorParameterValue _mid "Audio Secondary Color" :!class /Script/Engine.MaterialInstanceDynamic))
  (Variables|Default|SetAudio%20Secondary%20Color _returnvalue_3)
  (bind _output_get (Variables|Default|SetHue%20Shortest%20Path (Utilities|ShortestHuePath _hsv_x_2 _hsv_x)))
  (Rendering|Material|SetScalarParameterValueonMaterials (CustomCharacter|GetSkeletalMeshForPartOnPlayerPawn) "Shift Max" _output_get))

(fn ConstructionScript
  (|CallParentFunction))
