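; Body part modifier for the StarWalker Athena cosmetic, which sets up its S_FX_ShardLine effect component.
; After the post-res-in event runs its parent logic, it attaches that component to the part's skeletal mesh at FX_VortexCenter.
; Before attaching, it reads the r.SeparateTranslucency CVar and passes a 0 or 1 SepTrans_Check scalar to the component's materials, so they can adapt to that render setting.
; A failed CVar read only prints an error.
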(event Custom|Seperate%20Translucency%20Check
  (bind (_outvalue _returnvalue) (Fortnite|GetNumericCVar "r.SeparateTranslucency"))
  (if _returnvalue
    (Rendering|Material|SetScalarParameterValueonMaterials (Variables|Default|GetS_FX_ShardLine) "SepTrans_Check" (Utilities|Select :Index (Math|Float|Truncate _outvalue) :Option%200 0.0 :Option%201 1.0))
    (else
      (Development|PrintString "CVar Error"))))

(event EventOnPostResIn
  (|CallParentFunction)
  (CallFunction|SeperateTranslucencyCheck)
  (Transformation|AttachComponentToComponent (Variables|Default|GetS_FX_ShardLine) (CustomCharacter|GetSkeletalMeshForPartOnPlayerPawn) "FX_VortexCenter" "KeepRelative" "KeepRelative" "KeepRelative" true))

(fn ConstructionScript)
