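; Character part modifier for the Street Racer Drift Remix arms cosmetic that sets up mirrored lightning particle effects on the arms.
; After the parent's post-res-in logic, it configures the idle VFX component for the right side and attaches a second particle component to upperarm_l on the part's skeletal mesh.
; Vector parameters and LightningMesh_L/R emitter toggles give each component its side's offset and mesh, with an initial rotation on the left.
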(event EventOnPostResIn
  (bind _returnvalue (CallFunction|GetIdleVFXComponent))
  (bind _p_f_streetracer_driftremix_arml (Variables|Default|GetP_F_StreetRacer_DriftRemix_ArmL))
  (|CallParentFunction)
  (Effects|Components|ParticleSystem|SetVectorParameter _returnvalue "LocOffset" "-12.000000, 0,-10.000000")
  (Effects|Components|ParticleSystem|SetEmitterEnable _returnvalue "LightningMesh_L" false)
  (Effects|Components|ParticleSystem|SetEmitterEnable _returnvalue "LightningMesh_R" true)
  (Transformation|AttachComponentToComponent _p_f_streetracer_driftremix_arml (CustomCharacter|GetSkeletalMeshForPartOnPlayerPawn) "upperarm_l" "SnapToTarget" "KeepRelative" "KeepRelative" true)
  (Effects|Components|ParticleSystem|SetVectorParameter _p_f_streetracer_driftremix_arml "LocOffset" "12.000000, 0,10.000000")
  (Effects|Components|ParticleSystem|SetVectorParameter _p_f_streetracer_driftremix_arml "InitRot" "0.500000, 0,0.000000")
  (Effects|Components|ParticleSystem|SetEmitterEnable _p_f_streetracer_driftremix_arml "LightningMesh_L" true)
  (Effects|Components|ParticleSystem|SetEmitterEnable _p_f_streetracer_driftremix_arml "LightningMesh_R" false))

(fn ConstructionScript)
