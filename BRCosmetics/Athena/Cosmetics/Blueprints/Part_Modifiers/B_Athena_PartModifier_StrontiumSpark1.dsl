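; Cosmetic part modifier Blueprint for the StrontiumSpark1 character part that reacts to the player's kill stat.
; When the pawn or part changes, it runs the parent logic, then watches the kill stat tag on the associated player pawn.
; Each stat change passes the stat name, value and kill tag to an AudioStatWatcher function, apparently inherited, which by name handles audio feedback.
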(event EventOnPawnAndPartChanged
  (|CallParentFunction)
  (bind (_statname _statvalue) (Cosmetics|WatchCosmeticStat (CustomCharacter|GetAssociatedPlayerPawn) (Variables|Default|GetKillStatTag) 0 false)
    (:OnStatChanged
      (CallFunction|AudioStatWatcher _statname _statvalue (Variables|Default|GetKillStatTag)))))

(fn ConstructionScript
  (|CallParentFunction))
