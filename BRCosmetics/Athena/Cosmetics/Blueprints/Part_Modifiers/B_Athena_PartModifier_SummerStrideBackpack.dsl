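; Part modifier for the SummerStride back bling that makes its idle VFX react to the owning player's kills.
; A kill stat watcher fires KillReactiveEvent, which pulses the particle "KillEvent" bool on and off around a short delay.
; In the front end, a timeline paced by the kill event duration loops the reaction, with an attached sound outside the HeroSelect camera.
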(event EventOnPawnAndPartChanged
  (|CallParentFunction)
  (Cosmetics|WatchCosmeticStat (CustomCharacter|GetAssociatedPlayerPawn) "(TagName=\"StatManager.AthenaKills\")" 0 false
    (:OnStatChanged
      ;[1 0.63 0.4949] Listen for kills to call kill reactive event
      (if (> (stale EventOnStatChanged Stat%20Value) 0)
        (CallFunction|KillReactiveEvent)))))

(event EventOnStatChanged
  (params (State%20Name :type Name :default "None") (Stat%20Value :type Integer))
  ;[1 0.63 0.4949] Listen for kills to call kill reactive event
  (|CallParentFunction State%20Name Stat%20Value))

(event Custom|KillReactiveEvent
  (bind _returnvalue (CallFunction|GetIdleVFXComponent))
  ;[0.5309 0.5034 1] Kill reactive event send to the backpack vfx
  (Utilities|IsValid _returnvalue
    (:Is%20Valid
      (Effects|Components|ParticleSystem|SetBoolParameter _returnvalue "KillEvent" true)
      (if (Reactive|Frontend|ShouldDisplayFrontendReactivity 0 "Backpack")
        (if (not (Utilities|Enum|Equal(Enum) (FrontEndCamera|GetFrontEndCamera (PlayerController|LocalPlayerSubsystems|GetFortFrontEndContext (Fortnite|GetFirstLocalFortPlayerController true))) "HeroSelect"))
          (Audio|SpawnSoundAttached "/BRCosmetics/Sounds/Backblings/SummerStride/BackBling_SummerStride_Full_Cue.BackBling_SummerStride_Full_Cue" (Variables|/Script/Engine.Actor|GetRootComponent (CustomCharacter|GetAssociatedPlayerPawn))))
        (Utilities|FlowControl|Delay 0.5)
        (Effects|Components|ParticleSystem|SetBoolParameter _returnvalue "KillEvent" false)
        (else
          _)))))

(event EventOnPostResIn
  (bind _returnvalue (CallFunction|GetIdleVFXComponent))
  ;[1 0.599 0.599] Initial Setup
  (|CallParentFunction)
  (Utilities|IsValid _returnvalue
    (:Is%20Valid
      (Effects|Components|ParticleSystem|SetFloatParameter _returnvalue "EventDuration" (Variables|Default|GetKillEventDuration))
      ;[1 0.199 0.199] Setup looping reactivity in front end
      (if (Reactive|Frontend|ShouldDisplayFrontendReactivity 0 "Backpack")
        (Components|Timeline|SetPlayRate (Variables|B_Athena_PartModifier_SummerStrideBackpack|GetKillReactiveLoop) (Math|Float|SafeDivide 1.0 (* (Variables|Default|GetKillEventDuration) 2.0)))
        (Timeline|KillReactiveLoop PlayFromStart
          (:PlayReactiveEffects
            (CallFunction|KillReactiveEvent)))))))

(fn ConstructionScript
  (|CallParentFunction))
