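; Athena cosmetic part modifier for SunBeamQuest that gates stat reactivity on the variant of the SunbeamQuest backpack.
; After res-in, it reads the cosmetic loadout for the backpack's Parts channel against Stage1 and enables reactivity from the validity result.
; A helper pushes an integer stat value as a named parameter to audio components, noted as serving back bling and character skin.
; Stat changes pass straight to the parent class.
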(event EventOnStatChanged
  (params (State%20Name :type Name :default "None") (Stat%20Value :type Integer))
  (|CallParentFunction State%20Name Stat%20Value))

(event EventOnPostResIn
  (|CallParentFunction)
  ;[1 1 1] Read CIDs
  (bind (_index _bindexisvalid) (Class|FortCosmeticFunctionLibrary|GetVariantChannelActiveIndexForItem "/BRCosmetics/Athena/Items/Cosmetics/Backpacks/Backpack_SunbeamQuest.Backpack_SunbeamQuest" (CosmeticLoadout|GetCosmeticLoadout) "(TagName=\"Cosmetics.Variant.Channel.Parts\")" (Utilities|Array|MakeArray "(TagName=\"Cosmetics.Variant.Property.Stage1\")") false false))
  (Variables|Default|SetReactivityEnabled _bindexisvalid))

(fn UpdateAudioStatValue
  (params (AudioComponent :type Array%20of%20Audio%20Component%20Object%20References) (InName :type Name :default "None") (StatValue :type Integer))
  ;[1 1 1] used for backbling and character skin.
  (for _array_element AudioComponent
    (Audio|Parameter|SetIntegerParameter _array_element InName StatValue :!class /Script/AudioExtensions.AudioParameterControllerInterface)))

(fn ConstructionScript
  (|CallParentFunction))
