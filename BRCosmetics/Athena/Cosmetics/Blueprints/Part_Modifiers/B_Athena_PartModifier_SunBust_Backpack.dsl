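; Cosmetic part modifier for the SunBust backpack that plays its idle VFX while cycling a three-step "WordNum" particle parameter.
; Pawn jumps randomly trigger the effect, and in the front end a looping timer triggers it through UpdateFX.
; Each trigger activates the component, advances and wraps the counter, debug-prints it and sends it to the particles.
; Pawn or part changes reset the counter and clear the timer; res-in also resets the counter.
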(event EventOnDayPhaseChanged
  (params (CurrentDayPhase :type EFortDayPhase%20Enum :default "Morning") (PreviousDayPhase :type EFortDayPhase%20Enum :default "Morning") (bAtCreation :type Boolean)))

(event EventOnNewPlayerWeaponType
  (params (NewType :type EFortItemType%20Enum :default "WorldItem") (OldType :type EFortItemType%20Enum :default "WorldItem"))
  (|CallParentFunction NewType OldType) ; This node is disabled and will not be called.\nDrag off pins to build functionality.
)

(event EventOnWeaponChanged
  (params (NewWeapon :type Fort%20Weapon%20Object%20Reference) (OldWeapon :type Fort%20Weapon%20Object%20Reference)))

(event EventOnWeaponFired)

(event EventOnNewWeapons
  (params (WeaponsAsItems :type Array%20of%20Fort%20Item%20Object%20References)))

(event EventOnCurieStateAttached
  (params (Tag :type Gameplay%20Tag%20Structure%20%28by%20ref%29)))

(event EventOnCurieStateDetached
  (params (Tag :type Gameplay%20Tag%20Structure%20%28by%20ref%29)))

(event EventOnBeginSkydiving
  (|CallParentFunction) ; This node is disabled and will not be called.\nDrag off pins to build functionality.
)

(event EventOnPawnAndPartChanged
  (bind _timer (Variables|Default|GetTimer))
  (|CallParentFunction)
  (Variables|Default|SetWordNum -1)
  (if (Utilities|Time|IsValidTimerHandle _timer)
    (Utilities|Time|ClearandInvalidateTimerbyHandle _timer)))

(event EventOnPawnJumped
  (|CallParentFunction)
  (if (> (Math|Random|RandomFloat) 0.55)
    (Components|Activation|Activate (CallFunction|GetIdleVFXComponent) true)
    (Math|Integer|IncrementInt (Variables|Default|GetWordNum)
      (:%20%20
        (Utilities|FlowControl|Sequence
          (:then_0
            (if (> (Variables|Default|GetWordNum) 2)
              (Variables|Default|SetWordNum 0)))
          (:then_1
            (Development|PrintString (Utilities|String|ToString(Integer) (Variables|Default|GetWordNum)))
            (Effects|Components|ParticleSystem|SetIntParameter (CallFunction|GetIdleVFXComponent) "WordNum" (Variables|Default|GetWordNum))))))))

(event EventOnPostResIn
  (|CallParentFunction)
  (Variables|Default|SetWordNum -1)
  (if (CustomCosmeticModifier|IsinFrontEnd)
    (bind _returnvalue (Utilities|Time|SetTimerbyEvent (delegate Custom|UpdateFX) 2.0 true))
    (Variables|Default|SetTimer _returnvalue)))

(event Custom|UpdateFX
  (Components|Activation|Activate (CallFunction|GetIdleVFXComponent) true)
  (Math|Integer|IncrementInt (Variables|Default|GetWordNum)
    (:%20%20
      (Utilities|FlowControl|Sequence
        (:then_0
          (if (> (Variables|Default|GetWordNum) 2)
            (Variables|Default|SetWordNum 0)))
        (:then_1
          (Development|PrintString (Utilities|String|ToString(Integer) (Variables|Default|GetWordNum)))
          (Effects|Components|ParticleSystem|SetIntParameter (CallFunction|GetIdleVFXComponent) "WordNum" (Variables|Default|GetWordNum)))))))

(fn ConstructionScript
  (|CallParentFunction))
