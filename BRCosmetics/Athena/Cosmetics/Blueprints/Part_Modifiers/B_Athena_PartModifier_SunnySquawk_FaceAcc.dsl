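; Cosmetic part modifier Blueprint for the SunnySquawk face accessory that hides a material-driven element when the player's loadout has an extra part.
; After the post-res-in event and its parent logic, it checks the cosmetic loadout for a valid item in the ExtraPart slot.
; If one exists, it gets a dynamic material instance from the pawn's part mesh and switches on its "Hide Element 01" scalar.
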(event EventOnPostResIn
  (|CallParentFunction)
  (if (Utilities|IsValid (CosmeticLoadout|GetItemDefinitionfromLoadout (CosmeticLoadout|GetCosmeticLoadout) "(TagName=\"LoadoutSlot.ExtraPart\")") :!class /Script/Engine.KismetSystemLibrary)
    (Rendering|Material|SetScalarParameterValue (Class|BTechArtGlobalLibrary|GetDynamicMaterialInstance (CustomCharacter|GetSkeletalMeshForPartOnPlayerPawn)) "Hide Element 01" 1.0 :!class /Script/Engine.MaterialInstanceDynamic)))

(fn ConstructionScript
  (|CallParentFunction))
