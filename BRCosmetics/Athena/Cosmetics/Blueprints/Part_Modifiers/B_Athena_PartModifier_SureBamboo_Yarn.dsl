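; Cosmetic part modifier for the SureBamboo Yarn character, making the body material react to player kills.
; After post-res init and a toon LUT step, the frontend replays a timeline every 14 seconds as a preview.
; A kill-stat watcher triggers OnKill, whose timeline drives material Time through three phases, adding gradient glow from three kills, where the stored count stops updating.
; SetRandomCollageColors tints MiscOrTail but is uncalled here; SetMainColors is empty.
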(event Custom|FE_Reactivety
  (bind _returnvalue (CustomCharacter|GetAssociatedPlayerPawn))
  (bind _returnvalue_2 (Customization|GetSkeletalMeshforPartType _returnvalue "Body" :!class /Script/FortniteGame.FortPlayerPawn))
  ;[0 0.1411 1] FE reactive
  (bind (_direction _phase _gradientglow) (Timeline|TimelineFE PlayFromStart)
    (:Update
      (Rendering|Material|SetScalarParameterValueonMaterials _returnvalue_2 "Time" _phase)
      (Rendering|Material|SetScalarParameterValueonMaterials _returnvalue_2 "Gradient_Intensity" _gradientglow))))

(event EventOnPostResIn
  (bind _returnvalue (CustomCharacter|GetAssociatedPlayerPawn))
  (bind _returnvalue_2 (Customization|GetSkeletalMeshforPartType _returnvalue "Body" :!class /Script/FortniteGame.FortPlayerPawn))
  (|CallParentFunction)
  (Utilities|ToonFrontendLUTReplace
    (:Out
      (Utilities|FlowControl|Sequence
        (:then_0
          ;[0 0.1411 1] FE reactive
          (if (Reactive|Frontend|IsCosmeticObjectDisplayedinFrontendContext 0)
            (Utilities|FlowControl|Delay 3.5)
            (Utilities|Time|SetTimerbyEvent (delegate Custom|FE_Reactivety) 14.0 true)
            (Timeline|TimelineFE PlayFromStart)))
        (:then_1
          ;[0.2818 1 0.7667] On Kill
          (bind (_statname _statvalue) (Cosmetics|WatchCosmeticStat (CustomCharacter|GetAssociatedPlayerPawn) "(TagName=\"StatManager.AthenaKills\")" 0 false)
            (:OnStatChanged
              (if (not (>= (Variables|Default|GetKillCount) 3))
                (Variables|Default|SetKillCount _statvalue))
              (CallFunction|OnKill))))))))

(event Custom|OnKill
  (bind _returnvalue (CustomCharacter|GetAssociatedPlayerPawn))
  (bind _returnvalue_2 (Customization|GetSkeletalMeshforPartType _returnvalue "Body" :!class /Script/FortniteGame.FortPlayerPawn))
  (bind _killcount (Variables|Default|GetKillCount))
  (bind _returnvalue_3 (>= _killcount 3))
  ;[0.2818 1 0.7667] On Kill
  (bind (_direction _phase1 _phase2 _phase3 _gradientglow2 _gradientglow3) (Timeline|Timeline PlayFromStart)
    (:Update
      (switch int _killcount
        (:1
          _)
        (:2
          _)
        (:3
          (Rendering|Material|SetScalarParameterValueonMaterials _returnvalue_2 "Gradient_Intensity" (Math|Float|SelectFloat _gradientglow3 0.0 _returnvalue_3))))
      (Rendering|Material|SetScalarParameterValueonMaterials _returnvalue_2 "Time" (Math|Float|SelectFloat _phase3 (Math|Float|SelectFloat _phase2 _phase1 (== _killcount 2)) _returnvalue_3)))))

(fn SetMainColors)

(fn SetRandomCollageColors
  (bind _element (ArrayMacros|GetRandomArrayElement)
    (:%20%20
      (Rendering|Material|SetColorParameterValueonMaterials (Customization|GetSkeletalMeshforPartType (CustomCharacter|GetAssociatedPlayerPawn) "MiscOrTail" :!class /Script/FortniteGame.FortPlayerPawn) "Collage_Color" (* (Math|Conversions|ToLinearColor(Color) _element) (select (Reactive|Frontend|IsCosmeticObjectDisplayedinFrontendContext 0) "(R=0.700000,G=0.700000,B=0.700000,A=1.000000)" "(R=1.000000,G=1.000000,B=1.000000,A=1.000000)"))))))

(fn ConstructionScript
  (|CallParentFunction))
