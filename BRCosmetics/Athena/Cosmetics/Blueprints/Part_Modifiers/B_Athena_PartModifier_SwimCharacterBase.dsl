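; Base cosmetic part modifier for Swim characters that drives the idle VFX component from gameplay and loadout state.
; Setup uses the particle variant to gate kill reactivity and eye FX, and shows back FX when no real backpack is equipped.
; Kill-stat changes feed the VFX counter, while frontend previews increment it on a looping timer.
; A repeating timer passes mesh LOD to the effect, and targeting changes set separate local and remote ADS flags.
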(event EventOnPostResIn
  ;[0 0 0] On Post Res in
  (|CallParentFunction)
  (Utilities|FlowControl|Sequence
    (:then_0
      (CallFunction|SetupReactivity))
    (:then_1
      (CallFunction|SetupBackFXBool))
    (:then_2
      (if (Reactive|Frontend|IsCosmeticObjectDisplayedinFrontendContext 0)
        ;;[1 0.5572 0] Front end loop
        (Utilities|Time|SetTimerbyEvent (EventDispatchers|CreateEvent :!event Custom|OnFrontendReactivityPreview) 8.0 true false 4.0)
        (Utilities|FlowControl|Delay 4.0)
        (CallFunction|OnFrontendReactivityPreview) ; Calls the first reactivity loop immediately
        (else
          (bind (_statname _statvalue) (Cosmetics|WatchCosmeticStat (CustomCharacter|GetAssociatedPlayerPawn) "(TagName=\"StatManager.AthenaKills\")" 0 false)
            (:OnStatChanged
              (if (> _statvalue 0)
                (CallFunction|OnKillCountChanged _statvalue)))))))
    (:then_3
      (Utilities|Time|SetTimerbyEvent (EventDispatchers|CreateEvent :!event Custom|LODCheck) (Variables|Default|GetLODCheckInterval) true)
      (bind _ascustom_character_part_anim_instance (Utilities|Casting|CastToCustomCharacterPartAnimInstance (Components|SkeletalMesh|GetAnimInstance (Customization|GetSkeletalMeshforPartType (CustomCharacter|GetAssociatedPlayerPawn) (CustomCharacter|GetCharacterPartType) :!class /Script/FortniteGame.FortPlayerPawn))))
      (Variables|Default|SetAnimInstance _ascustom_character_part_anim_instance))) ; 0 - Reactivity Setup\n1 - BackFX\n2 - Frontend\n3 - Timer for LOD Check
)

(event Custom|LODCheck
  ;[1 1 1] LOD Check
  (Effects|Components|ParticleSystem|SetIntParameter (CallFunction|GetIdleVFXComponent) "SkeletalMeshCurrentLOD" (Variables|/Script/FortniteGame.CustomCharacterPartAnimInstance|GetCurrentLOD (Variables|Default|GetAnimInstance))))

(event EventOnTargetingOrFirstPersonCameraChanged
  (bind _returnvalue (CustomCharacter|GetAssociatedPlayerPawn))
  (bind _returnvalue_2 (CallFunction|GetIdleVFXComponent))
  (bind _bisadsing (Variables|Default|GetbIsADSing))
  ;[0 0 0] ADS
  (|CallParentFunction)
  (Variables|Default|SetbIsADSing (Variables|/Script/FortniteGame.FortPlayerPawn|GetbIsTargeting _returnvalue))
  (if (Pawn|IsLocallyControlled _returnvalue)
    (Effects|Components|ParticleSystem|SetBoolParameter _returnvalue_2 "bIsADS" _bisadsing)
    (else
      (Effects|Components|ParticleSystem|SetBoolParameter _returnvalue_2 "bIsADSOther" _bisadsing))))

(event Custom|OnKillCountChanged
  (params (KillCount :type Integer))
  ;[0 0 0] Reactivity Implementation
  (bind _output_get (Variables|Default|SetKill%20Count KillCount))
  (if (Variables|Default|GetbIsReactivityEnabled)
    (Utilities|FlowControl|Sequence
      (:then_0
        (Effects|Components|ParticleSystem|SetIntParameter (CallFunction|GetIdleVFXComponent) "KillReactivityCount" _output_get)))))

(event Custom|OnFrontendReactivityPreview
  ;[0 0.1333 1] Reactivity for frontend
  (bind _result (Math|Integer|IncrementInt (Variables|Default|GetKill%20Count))
    (:%20%20
      (CallFunction|OnKillCountChanged _result))))

(fn B_Athena_PartModifier_GreatPoolSwim_Character_AutoGenFunc)

(fn SetupBackFXBool
  (bind _returnvalue (CallFunction|GetIdleVFXComponent))
  (Utilities|IsValid _returnvalue
    (:Is%20Valid
      (bind _asniagara_particle_system_component (Utilities|Casting|CastToNiagaraComponent _returnvalue))
      (Niagara|SetNiagaraVariable(Bool) _asniagara_particle_system_component "bShowBackFX" (or (not (Utilities|IsValid (CosmeticLoadout|GetItemDefinitionfromLoadout (CosmeticLoadout|GetCosmeticLoadout) "(TagName=\"LoadoutSlot.Backpack\")") :!class /Script/Engine.KismetSystemLibrary)) (== (CosmeticLoadout|GetItemDefinitionfromLoadout (CosmeticLoadout|GetCosmeticLoadout) "(TagName=\"LoadoutSlot.Backpack\")") "/Game/Athena/Items/Cosmetics/Backpacks/BID_STWHeroNoDefaultBackpack.BID_STWHeroNoDefaultBackpack"))))))

(fn SetupReactivity
  (bind _returnvalue (CallFunction|GetIdleVFXComponent))
  (bind _asfort_player_pawn (Utilities|Casting|CastToFortPlayerPawn (Actor|GetOwner)))
  (bind _index (CosmeticLoadout|GetVariantChannelActiveIndexforSlotActor "(TagName=\"LoadoutSlot.Character\")" _asfort_player_pawn "(TagName=\"Cosmetics.Variant.Channel.Particle\")" (Utilities|Array|MakeArray "(TagName=\"Cosmetics.Variant.Property.Particle1\")" "(TagName=\"Cosmetics.Variant.Property.Particle2\")")))
  (bind _output_get (Variables|Default|SetbIsReactivityEnabled (not (Math|Conversions|ToBoolean(Integer) _index))))
  (Utilities|IsValid _returnvalue
    (:Is%20Valid
      (bind _asniagara_particle_system_component (Utilities|Casting|CastToNiagaraComponent _returnvalue))
      (Niagara|SetNiagaraVariable(Bool) _asniagara_particle_system_component "bShowEyeFX" _output_get))))

(fn ConstructionScript
  (|CallParentFunction))
