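; Character part modifier for the TacticalRedGlow pack cosmetic that toggles the blade in its idle VFX.
; On weapon change and skydive end, it hides "ShowBlade" if the part's anim instance reports a pickaxe tagged TacticalRedPunkFemale, otherwise shows it.
; Skydiving forces the blade visible and suppresses weapon-change updates.
; A short BeginPlay delay clears the OnBegin guard to avoid a null cast, and debug prints remain.
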(event EventBeginPlay
  (Utilities|FlowControl|Delay 0.1) ; Need slight delay to avoid getting a "None" call on the below Cast on startup
  (Variables|Default|SetOnBegin false))

(event Collision|EventActorBeginOverlap
  (params (OtherActor :type Actor%20Object%20Reference)))

(event EventTick
  (params (DeltaSeconds :type Float%20%28single-precision%29)))

(event EventOnWeaponChanged
  (params (NewWeapon :type Fort%20Weapon%20Object%20Reference) (OldWeapon :type Fort%20Weapon%20Object%20Reference))
  (bind _returnvalue (CallFunction|GetIdleVFXComponent))
  (|CallParentFunction NewWeapon OldWeapon)
  (if (Variables|Default|GetOnBegin)
    _
    (elif (Variables|Default|GetIsSkyDiving)
      _
      (else
        (bind _pickaxeequipped (CallFunction|AnimationCheck))
        (if _pickaxeequipped
          (Effects|Components|ParticleSystem|SetBoolParameter _returnvalue "ShowBlade" false)
          (Development|PrintString "BladeisTrue")
          (else
            (Effects|Components|ParticleSystem|SetBoolParameter _returnvalue "ShowBlade" true)
            (Development|PrintString "BladeisFalse")))))))

(event EventOnPostResIn
  (|CallParentFunction))

(event EventOnBeginSkydiving
  (|CallParentFunction)
  (Variables|Default|SetIsSkyDiving true)
  (Effects|Components|ParticleSystem|SetBoolParameter (CallFunction|GetIdleVFXComponent) "ShowBlade" true))

(event EventOnEndSkydiving
  (bind _returnvalue (CallFunction|GetIdleVFXComponent))
  (|CallParentFunction)
  (Variables|Default|SetIsSkyDiving false)
  (bind _ascustom_character_part_anim_instance (Utilities|Casting|CastToCustomCharacterPartAnimInstance (Components|SkeletalMesh|GetAnimInstance (Customization|GetSkeletalMeshforPartType (CustomCharacter|GetAssociatedPlayerPawn) (CustomCharacter|GetCharacterPartType) :!class /Script/FortniteGame.FortPlayerPawn))))
  (if (CosmeticTags|PickaxeItemDefinitionHasAnyMetaTag _ascustom_character_part_anim_instance (Utilities|Array|MakeArray "(TagName=\"Cosmetics.MetaTags.TacticalRedPunkFemale\")") true)
    (Effects|Components|ParticleSystem|SetBoolParameter _returnvalue "ShowBlade" false)
    (else
      (Effects|Components|ParticleSystem|SetBoolParameter _returnvalue "ShowBlade" true))))

(fn AnimationCheck
  (bind _ascustom_character_part_anim_instance (Utilities|Casting|CastToCustomCharacterPartAnimInstance (Components|SkeletalMesh|GetAnimInstance (Customization|GetSkeletalMeshforPartType (CustomCharacter|GetAssociatedPlayerPawn) (CustomCharacter|GetCharacterPartType) :!class /Script/FortniteGame.FortPlayerPawn))))
  (return (CosmeticTags|PickaxeItemDefinitionHasAnyMetaTag _ascustom_character_part_anim_instance (Utilities|Array|MakeArray "(TagName=\"Cosmetics.MetaTags.TacticalRedPunkFemale\")") true)))

(fn ConstructionScript
  (|CallParentFunction))
