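; Part modifier Blueprint for the Tactical Woodland Blue backpack cosmetic that shows the player's kill count on its material.
; After post-res-in and the parent call, it watches the AthenaKills cosmetic stat on the associated player pawn.
; Each stat change writes the value to the part mesh's Counter_InputNumber parameter and game time to TimeStart, leaving display and timed response to the material.
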(event EventOnPostResIn
  (bind _returnvalue (CustomCharacter|GetSkeletalMeshForPartOnPlayerPawn))
  ;[1 1 1] Kill Counter Tracking
  (|CallParentFunction)
  (bind (_statname _statvalue) (Cosmetics|WatchCosmeticStat (CustomCharacter|GetAssociatedPlayerPawn) "(TagName=\"StatManager.AthenaKills\")")
    (:OnStatChanged
      (Rendering|Material|SetScalarParameterValueonMaterials _returnvalue "Counter_InputNumber" (Math|Conversions|ToFloat(Integer) _statvalue))
      (Rendering|Material|SetScalarParameterValueonMaterials _returnvalue "TimeStart" (Utilities|Time|GetGameTimeinSeconds)))))

(fn ConstructionScript)
