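; Tailor body part modifier that makes a character's materials react to the player's match stats.
; On post-res-in, it watches kills, damage dealt and received, and health and shield gains.
; Kills drive an Elimination timeline, while other events send event-specific colors to Play-Layer0/1/2.
; InUse flags choose among these three timeline-animated material layers so colored effects can overlap on the part's skeletal mesh.
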(event Custom|Play-Layer2
  (params (SetColor :type Linear%20Color%20Structure) (SetMesh :type Skeletal%20Mesh%20Component%20Object%20Reference))
  (Components|Timeline|SetPlayRate (Variables|B_Athena_PartModifier_Tailor_Body|GetLayer2-TL) 0.5)
  (Variables|Default|SetLayer2-InUse true)
  (Utilities|FlowControl|Sequence
    (:then_0
      (bind (_direction _layer2_start) (Timeline|Layer2-TL Stop)
        (:Update
          (Rendering|Material|SetScalarParameterValueonMaterials SetMesh "Layer2-Start" _layer2_start))
        (:Finished
          (Variables|Default|SetLayer2-InUse false))))
    (:then_1
      (Rendering|Material|SetScalarParameterValueonMaterials SetMesh "Layer2-Start" 1.8)
      (Rendering|Material|SetVectorParameterValueonMaterials SetMesh "Layer2-Color" (Math|Conversions|ToVector(LinearColor) SetColor))
      (Timeline|Layer2-TL PlayFromStart))))

(event EventOnPostResIn
  (bind _returnvalue (CustomCharacter|GetAssociatedPlayerPawn))
  (bind _returnvalue_2 (CustomCharacter|GetCharacterPartType))
  (bind _returnvalue_3 (CustomCharacter|GetAssociatedPlayerPawn))
  (bind _returnvalue_4 (CustomCharacter|GetCharacterPartType))
  (bind _returnvalue_5 (Customization|GetSkeletalMeshforPartType _returnvalue_3 _returnvalue_4 :!class /Script/FortniteGame.FortPlayerPawn))
  (bind _returnvalue_6 (CustomCharacter|GetCharacterPartType))
  (bind _returnvalue_7 (CustomCharacter|GetAssociatedPlayerPawn))
  (bind _returnvalue_8 (CustomCharacter|GetCharacterPartType))
  (bind _returnvalue_9 (Customization|GetSkeletalMeshforPartType _returnvalue_7 _returnvalue_8 :!class /Script/FortniteGame.FortPlayerPawn))
  (bind _returnvalue_10 (CustomCharacter|GetAssociatedPlayerPawn))
  (bind _returnvalue_11 (CustomCharacter|GetCharacterPartType))
  (bind _returnvalue_12 (Customization|GetSkeletalMeshforPartType _returnvalue_10 _returnvalue_11 :!class /Script/FortniteGame.FortPlayerPawn))
  (bind _returnvalue_13 (CustomCharacter|GetAssociatedPlayerPawn))
  (bind _returnvalue_14 (Customization|GetSkeletalMeshforPartType _returnvalue_13 _returnvalue_2 :!class /Script/FortniteGame.FortPlayerPawn))
  (bind _returnvalue_15 (Customization|GetSkeletalMeshforPartType _returnvalue _returnvalue_6 :!class /Script/FortniteGame.FortPlayerPawn))
  (bind _health_color (Variables|Default|GetHealth-Color))
  (bind _shield_color (Variables|Default|GetShield-Color))
  (bind _damagedealt_color (Variables|Default|GetDamageDealt-Color))
  (bind _damaged_color (Variables|Default|GetDamaged-Color))
  (|CallParentFunction)
  (Utilities|FlowControl|Sequence
    (:then_0
      ;[1 0.6239 0] Elimination
      (bind (_statname _statvalue) (Cosmetics|WatchCosmeticStat _returnvalue_3 "(TagName=\"StatManager.AthenaKills\")")
        (:OnStatChanged
          (if (> _statvalue 0)
            (bind (_direction _elim_start _elim_exit) (Timeline|Elimination PlayFromStart)
              (:Update
                (Rendering|Material|SetScalarParameterValueonMaterials _returnvalue_5 "Elimination-Start" _elim_start)
                (Rendering|Material|SetScalarParameterValueonMaterials _returnvalue_5 "Elimination-Exit" _elim_exit))
              (:Finished
                (Rendering|Material|SetScalarParameterValueonMaterials _returnvalue_5 "Elimination-Start" 0.0)
                (Rendering|Material|SetScalarParameterValueonMaterials _returnvalue_5 "Elimination-Exit" 0.0)))))))
    (:then_1
      ;[1 0.007073 0] Take Damage
      (bind (_statname_2 _statvalue_2) (Cosmetics|WatchCosmeticStat _returnvalue_7 "(TagName=\"StatManager.DamageReceived\")")
        (:OnStatChanged
          (if (> _statvalue_2 0)
            (if (Variables|Default|GetLayer0-InUse)
              (if (Variables|Default|GetLayer1-InUse)
                (if (not (Variables|Default|GetLayer2-InUse))
                  (CallFunction|Play-Layer2 _damaged_color _returnvalue_9))
                (else
                  (CallFunction|Play-Layer1 _damaged_color _returnvalue_9))))
            (CallFunction|Play-Layer0 _damaged_color _returnvalue_9)))) ; Damaged
    )
    (:then_2
      ;[0.08239 0.3963 1] Damage Dealt
      (bind (_statname_3 _statvalue_3) (Cosmetics|WatchCosmeticStat _returnvalue_10 "(TagName=\"StatManager.DamageDealt\")")
        (:OnStatChanged
          (if (> _statvalue_3 0)
            (if (Variables|Default|GetLayer0-InUse)
              (if (Variables|Default|GetLayer1-InUse)
                (if (not (Variables|Default|GetLayer2-InUse))
                  (CallFunction|Play-Layer2 _damagedealt_color _returnvalue_12))
                (else
                  (CallFunction|Play-Layer1 _damagedealt_color _returnvalue_12))))
            (CallFunction|Play-Layer0 _damagedealt_color _returnvalue_12)))) ; Temp: Damage dealt
    )
    (:then_3
      ;[0.008735 1 0] Gain Health
      (bind (_statname_4 _statvalue_4) (Cosmetics|WatchCosmeticStat _returnvalue "(TagName=\"StatManager.GainedHealthTimes\")")
        (:OnStatChanged
          (if (> _statvalue_4 0)
            (if (Variables|Default|GetLayer0-InUse)
              (if (Variables|Default|GetLayer1-InUse)
                (if (not (Variables|Default|GetLayer2-InUse))
                  (CallFunction|Play-Layer2 _health_color _returnvalue_15))
                (else
                  (CallFunction|Play-Layer1 _health_color _returnvalue_15))))
            (CallFunction|Play-Layer0 _health_color _returnvalue_15)))))
    (:then_4
      ;[0.008735 1 0] Gain Shield
      (bind (_statname_5 _statvalue_5) (Cosmetics|WatchCosmeticStat _returnvalue_13 "(TagName=\"StatManager.GainedShieldTimes\")")
        (:OnStatChanged
          (if (> _statvalue_5 0)
            (if (Variables|Default|GetLayer0-InUse)
              (if (Variables|Default|GetLayer1-InUse)
                (if (not (Variables|Default|GetLayer2-InUse))
                  (CallFunction|Play-Layer2 _shield_color _returnvalue_14))
                (else
                  (CallFunction|Play-Layer1 _shield_color _returnvalue_14))))
            (CallFunction|Play-Layer0 _shield_color _returnvalue_14)))))))

(event Custom|Play-Layer1
  (params (SetColor :type Linear%20Color%20Structure) (SetMesh :type Skeletal%20Mesh%20Component%20Object%20Reference))
  (Components|Timeline|SetPlayRate (Variables|B_Athena_PartModifier_Tailor_Body|GetLayer1-TL) 0.5)
  (Variables|Default|SetLayer1-InUse true)
  (Utilities|FlowControl|Sequence
    (:then_0
      (bind (_direction _layer1_start) (Timeline|Layer1-TL Stop)
        (:Update
          (Rendering|Material|SetScalarParameterValueonMaterials SetMesh "Layer1-Start" _layer1_start))
        (:Finished
          (Variables|Default|SetLayer1-InUse false))))
    (:then_1
      (Rendering|Material|SetScalarParameterValueonMaterials SetMesh "Layer1-Start" 1.8)
      (Rendering|Material|SetVectorParameterValueonMaterials SetMesh "Layer1-Color" (Math|Conversions|ToVector(LinearColor) SetColor))
      (Timeline|Layer1-TL PlayFromStart))))

(event Custom|Play-Layer0
  (params (SetColor :type Linear%20Color%20Structure) (SetMesh :type Skeletal%20Mesh%20Component%20Object%20Reference))
  (Components|Timeline|SetPlayRate (Variables|B_Athena_PartModifier_Tailor_Body|GetLayer0-TL) 0.5)
  (Variables|Default|SetLayer0-InUse true)
  (Utilities|FlowControl|Sequence
    (:then_0
      (bind (_direction _layer0_start) (Timeline|Layer0-TL Stop)
        (:Update
          (Rendering|Material|SetScalarParameterValueonMaterials SetMesh "Layer0-Start" _layer0_start))
        (:Finished
          (Variables|Default|SetLayer0-InUse false))))
    (:then_1
      (Rendering|Material|SetScalarParameterValueonMaterials SetMesh "Layer0-Start" 1.8)
      (Rendering|Material|SetVectorParameterValueonMaterials SetMesh "Layer0-Color" (Math|Conversions|ToVector(LinearColor) SetColor))
      (Timeline|Layer0-TL PlayFromStart))))

(fn ConstructionScript)
