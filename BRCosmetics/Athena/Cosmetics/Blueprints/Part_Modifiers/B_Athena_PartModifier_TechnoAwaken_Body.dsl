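; Cosmetic part modifier for the TechnoAwaken body that drives its idle Niagara VFX in reaction to kills and emotes.
; After post-res-in, it watches the pawn's AthenaKills cosmetic stat; each change plays the Reactive timeline from start, feeding its track into the "Kill" float parameter.
; Emote begin and end events toggle the bIsEmote bool on the same effect.
; A debug PrintString still fires on each stat change.
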(event EventOnPostResIn
  (|CallParentFunction)
  (Cosmetics|WatchCosmeticStat (CustomCharacter|GetAssociatedPlayerPawn) "(TagName=\"StatManager.AthenaKills\")" 0 false
    (:OnStatChanged
      (Development|PrintString "KillAthene" true true "(R=1.000000,G=0.000000,B=0.311551,A=1.000000)")
      (bind _asniagara_particle_system_component (Utilities|Casting|CastToNiagaraComponent (CallFunction|GetIdleVFXComponent)))
      (bind (_direction _newtrack_0) (Timeline|Reactive PlayFromStart)
        (:Update
          (Niagara|SetNiagaraVariable(Float) _asniagara_particle_system_component "Kill" _newtrack_0))))))

(event EventOnBeginEmote
  (params (MontageItemDef :type Fort%20Item%20Definition%20Object%20Reference) (PawnEmoting :type Fort%20Pawn%20Object%20Reference))
  (bind _asniagara_particle_system_component (Utilities|Casting|CastToNiagaraComponent (CallFunction|GetIdleVFXComponent)))
  (Niagara|SetNiagaraVariable(Bool) _asniagara_particle_system_component "bIsEmote" true))

(event EventOnEndEmote
  (params (MontageItemDef :type Fort%20Item%20Definition%20Object%20Reference) (PawnEmoting :type Fort%20Pawn%20Object%20Reference))
  (bind _asniagara_particle_system_component (Utilities|Casting|CastToNiagaraComponent (CallFunction|GetIdleVFXComponent)))
  (Niagara|SetNiagaraVariable(Bool) _asniagara_particle_system_component "bIsEmote" false))

(fn ConstructionScript
  (|CallParentFunction))
