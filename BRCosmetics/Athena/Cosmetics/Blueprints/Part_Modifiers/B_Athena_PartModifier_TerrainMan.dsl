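; Cosmetic part modifier for the Terrain Man back bling that spawns its Niagara effect on the backpack.
; After res-in in the front end, it waits briefly and attaches the effect to the "pack" socket of the part's skeletal mesh.
; Any stat change with a positive value spawns the same effect one tick later, without checking the stat name or an existing instance.
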(event EventOnPostResIn
  (|CallParentFunction)
  (if (CustomCosmeticModifier|IsinFrontEnd)
    (Utilities|FlowControl|Delay 2.0)
    (PartModifier|SpawnSystem|SpawnSystemAttachedPartModifier "/BRCosmetics/Accessories/FORT_Backpacks/Backpack_M_MED_Terrain_Man/Effects/NS_Backpack_Terrain_Man.NS_Backpack_Terrain_Man" (CustomCharacter|GetSkeletalMeshForPartOnPlayerPawn) "pack" "0, 0, 0" "0, 0, 0" "SnapToTarget"
      (:Success
        _)
      (:Failure
        _))
    (Utilities|FlowControl|Delay 5.0)))

(event EventOnStatChanged
  (params (State%20Name :type Name :default "None") (Stat%20Value :type Integer))
  (|CallParentFunction State%20Name Stat%20Value)
  (if (> Stat%20Value 0)
    (Utilities|FlowControl|DelayUntilNextTick)
    (PartModifier|SpawnSystem|SpawnSystemAttachedPartModifier "/BRCosmetics/Accessories/FORT_Backpacks/Backpack_M_MED_Terrain_Man/Effects/NS_Backpack_Terrain_Man.NS_Backpack_Terrain_Man" (CustomCharacter|GetSkeletalMeshForPartOnPlayerPawn) "pack" "0, 0, 0" "0, 0, 0" "SnapToTarget")))

(fn ConstructionScript
  (|CallParentFunction))
