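; Cosmetic part modifier Blueprint for the TerrianMan body part that feeds a time value into its material.
; Every tick it gets the dynamic material instance on the part's skeletal mesh and sets the CurrentTime scalar parameter.
; In the front end the value is game time scaled to about one third; otherwise it is the current time of day.
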(event EventTick
  (params (DeltaSeconds :type Float%20%28single-precision%29))
  (|CallParentFunction DeltaSeconds)
  (Rendering|Material|SetScalarParameterValue (Material|GetMID (CustomCharacter|GetSkeletalMeshForPartOnPlayerPawn)) "CurrentTime" (Math|Float|SelectFloat (* (Utilities|Time|GetTimeSeconds) 0.33333) (TimeOfDay|GetTimeOfDay :!class /Script/FortniteGame.FortKismetLibrary) (CustomCosmeticModifier|IsinFrontEnd)) :!class /Script/Engine.MaterialInstanceDynamic))

(fn ConstructionScript
  (|CallParentFunction))
