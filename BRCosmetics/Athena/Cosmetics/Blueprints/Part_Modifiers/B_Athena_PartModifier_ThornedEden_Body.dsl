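; Body part modifier for the Thorned Eden cosmetic that pulses parameters on the character's idle particle effect in response to movement.
; Jump and landing events briefly raise JumpBurst and LandBurst, then zero them after short delays.
; The movement update checks horizontal pawn speed against a sprint range and has a DoOnce-gated SprintBurst pulse, with no reset shown.
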(event EventOnPawnJumped
  (bind _returnvalue (CallFunction|GetIdleVFXComponent))
  (|CallParentFunction)
  (Effects|Components|ParticleSystem|SetFloatParameter _returnvalue "JumpBurst" 1.0)
  (Utilities|FlowControl|Delay 0.2)
  (Effects|Components|ParticleSystem|SetFloatParameter _returnvalue "JumpBurst" 0.0))

(event EventOnPawnLanded
  (bind _returnvalue (CallFunction|GetIdleVFXComponent))
  (|CallParentFunction)
  (Effects|Components|ParticleSystem|SetFloatParameter _returnvalue "LandBurst" 1.0)
  (Utilities|FlowControl|Delay 0.1)
  (Effects|Components|ParticleSystem|SetFloatParameter _returnvalue "LandBurst" 0.0))

(event EventOnCharacterMovementUpdated
  (params (DeltaSeconds :type Float%20%28single-precision%29) (OldLocation :type Vector :default "0, 0, 0") (OldVelocity :type Vector :default "0, 0, 0"))
  (bind _returnvalue (CallFunction|GetIdleVFXComponent))
  (|CallParentFunction DeltaSeconds OldLocation OldVelocity)
  (if (and (> (Math|Vector|VectorLength (* (Transformation|GetVelocity (CustomCharacter|GetAssociatedPlayerPawn)) "1.000000,1.000000,0.000000")) 6e+02) (< (Math|Vector|VectorLength (* (Transformation|GetVelocity (CustomCharacter|GetAssociatedPlayerPawn)) "1.000000,1.000000,0.000000")) 9e+02))
    _)
  (Utilities|FlowControl|DoOnce false
    (:Completed
      (Effects|Components|ParticleSystem|SetFloatParameter _returnvalue "SprintBurst" 1.0)
      (Utilities|FlowControl|Delay 0.1)
      (Effects|Components|ParticleSystem|SetFloatParameter _returnvalue "SprintBurst" 0.0))))

(fn ConstructionScript
  (|CallParentFunction))
