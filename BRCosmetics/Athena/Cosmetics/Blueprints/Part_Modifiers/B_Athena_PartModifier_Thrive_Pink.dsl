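; Cosmetic part modifier for the Thrive Pink outfit that makes its material and idle VFX react to the player's combat.
; Damage-dealt stat changes raise intensity one step, capped at one, while kills set it to full.
; Intensity decays linearly since the last hit; Blueprint passes previous value, target and game time as parameters instead of ticking.
; In the front end, a preview flag and timeline drive a debug reactive level.
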(event EventOnPostResIn
  (bind _calculatedintensity (CallFunction|CalculateCurrentReactiveIntensity))
  (bind _returnvalue (CallFunction|GetIdleVFXComponent))
  (bind _returnvalue_2 (CallFunction|GetIdleVFXComponent))
  (bind _returnvalue_3 (CallFunction|GetIdleVFXComponent))
  (bind _returnvalue_4 (CustomCharacter|GetSkeletalMeshForPartOnPlayerPawn))
  (bind _returnvalue_5 (CustomCharacter|GetSkeletalMeshForPartOnPlayerPawn))
  (bind _returnvalue_6 (CallFunction|GetIdleVFXComponent))
  (bind _returnvalue_7 (CallFunction|GetIdleVFXComponent))
  (|CallParentFunction)
  (Utilities|FlowControl|Sequence
    (:then_0
      ;[1 1 1] Initialize material
      (Rendering|Material|SetScalarParameterValueonMaterials _returnvalue_5 "Thrive_EffectDecayRate" (Variables|Default|GetReactiveDecayRate))
      (Rendering|Material|SetScalarParameterValueonMaterials _returnvalue_5 "Thrive_MaxLevel" (Math|Conversions|ToFloat(Integer) (Variables|Default|GetReactivityLevels)))
      (Rendering|Material|SetScalarParameterValueonMaterials _returnvalue_5 "Thrive_EffectLevel" 0.0)
      (Effects|Components|ParticleSystem|SetFloatParameter _returnvalue_3 "Thrive_EffectDecayRate" (Variables|Default|GetReactiveDecayRate))
      (Effects|Components|ParticleSystem|SetFloatParameter _returnvalue_3 "Thrive_EffectLevel" 0.0)
      (Effects|Components|ParticleSystem|SetFloatParameter _returnvalue_3 "Thrive_MaxLevel" (Math|Conversions|ToFloat(Integer) (Variables|Default|GetReactivityLevels)))
      (Variables|Default|SetReactiveIntensityStepAmount (/ 1.0 (Math|Conversions|ToFloat(Integer) (Variables|Default|GetReactivityLevels)))))
    (:then_1
      ;[1 1 1] Damage
      (Cosmetics|WatchCosmeticStat (CustomCharacter|GetAssociatedPlayerPawn) "(TagName=\"StatManager.DamageDealt\")"
        (:OnStatChanged
          (Rendering|Material|SetScalarParameterValueonMaterials (CustomCharacter|GetSkeletalMeshForPartOnPlayerPawn) "Thrive_GameTimeFromBP" (Utilities|Time|GetGameTimeinSeconds))
          (Effects|Components|ParticleSystem|SetFloatParameter (CallFunction|GetIdleVFXComponent) "Thrive_GameTimeFromBP" (Utilities|Time|GetGameTimeinSeconds))
          (Rendering|Material|SetScalarParameterValueonMaterials (CustomCharacter|GetSkeletalMeshForPartOnPlayerPawn) "Thrive_Previous_Intensity" _calculatedintensity)
          (bind _output_get (Variables|Default|SetCurrentIntensity (Math|Float|Min(Float) (+ _calculatedintensity (Variables|Default|GetReactiveIntensityStepAmount)) 1.0)))
          (Rendering|Material|SetScalarParameterValueonMaterials (CustomCharacter|GetSkeletalMeshForPartOnPlayerPawn) "Thrive_Target_Intensity" _output_get)
          (Variables|Default|SetLastHitTime (Utilities|Time|GetGameTimeinSeconds))
          (Effects|Components|ParticleSystem|SetFloatParameter _returnvalue "Thrive_Previous_Intensity" (CallFunction|CalculateCurrentReactiveIntensity))
          (Effects|Components|ParticleSystem|SetFloatParameter _returnvalue "Thrive_Target_Intensity" (Variables|Default|GetCurrentIntensity)))))
    (:then_2
      ;[1 1 1] Kills
      (Cosmetics|WatchCosmeticStat (CustomCharacter|GetAssociatedPlayerPawn) "(TagName=\"StatManager.AthenaKills\")"
        (:OnStatChanged
          (Rendering|Material|SetScalarParameterValueonMaterials (CustomCharacter|GetSkeletalMeshForPartOnPlayerPawn) "Thrive_GameTimeFromBP" (Utilities|Time|GetGameTimeinSeconds))
          (Effects|Components|ParticleSystem|SetFloatParameter (CallFunction|GetIdleVFXComponent) "Thrive_GameTimeFromBP" (Utilities|Time|GetGameTimeinSeconds))
          (Rendering|Material|SetScalarParameterValueonMaterials (CustomCharacter|GetSkeletalMeshForPartOnPlayerPawn) "Thrive_Previous_Intensity" (CallFunction|CalculateCurrentReactiveIntensity))
          (bind _output_get_2 (Variables|Default|SetCurrentIntensity 1.0))
          (Rendering|Material|SetScalarParameterValueonMaterials (CustomCharacter|GetSkeletalMeshForPartOnPlayerPawn) "Thrive_Target_Intensity" _output_get_2)
          (Variables|Default|SetLastHitTime (Utilities|Time|GetGameTimeinSeconds))
          (Effects|Components|ParticleSystem|SetFloatParameter _returnvalue_2 "Thrive_Previous_Intensity" (CallFunction|CalculateCurrentReactiveIntensity))
          (Effects|Components|ParticleSystem|SetFloatParameter _returnvalue_2 "Thrive_Target_Intensity" (Variables|Default|GetCurrentIntensity)))))
    (:then_3
      ;[1 1 1] Set variable initial state
      (Variables|Default|SetCurrentIntensity 0.0)
      (Variables|Default|SetLastHitTime -1e+03)
      (Rendering|Material|SetScalarParameterValueonMaterials (CustomCharacter|GetSkeletalMeshForPartOnPlayerPawn) "Thrive_Previous_Intensity" 0.0)
      (Rendering|Material|SetScalarParameterValueonMaterials (CustomCharacter|GetSkeletalMeshForPartOnPlayerPawn) "Thrive_Target_Intensity" 0.0)
      (Effects|Components|ParticleSystem|SetFloatParameter _returnvalue_6 "Thrive_Previous_Intensity" 0.0)
      (Effects|Components|ParticleSystem|SetFloatParameter _returnvalue_6 "Thrive_Target_Intensity" 0.0))
    (:then_4
      (if (CustomCosmeticModifier|IsinFrontEnd)
        (Rendering|Material|SetScalarParameterValueonMaterials _returnvalue_4 "FE_Preview" 1.0)
        (Effects|Components|ParticleSystem|SetFloatParameter _returnvalue_7 "FE_Preview" 1.0)
        (bind (_direction _reactive_level) (Timeline|Reactive%20Level Play)
          (:Update
            (Rendering|Material|SetScalarParameterValueonMaterials _returnvalue_4 "Debug_Reactive_Level" _reactive_level)
            (Effects|Components|ParticleSystem|SetFloatParameter (CallFunction|GetIdleVFXComponent) "Debug_Reactive_Level" _reactive_level)))
        (else
          (Rendering|Material|SetScalarParameterValueonMaterials _returnvalue_4 "FE_Preview" 0.0)
          (Effects|Components|ParticleSystem|SetFloatParameter _returnvalue_7 "FE_Preview" 0.0))))))

(event EventOnDayPhaseChanged
  (params (CurrentDayPhase :type EFortDayPhase%20Enum :default "Morning") (PreviousDayPhase :type EFortDayPhase%20Enum :default "Morning") (bAtCreation :type Boolean)))

(event EventOnNewPlayerWeaponType
  (params (NewType :type EFortItemType%20Enum :default "WorldItem") (OldType :type EFortItemType%20Enum :default "WorldItem"))
  (|CallParentFunction NewType OldType) ; This node is disabled and will not be called.\nDrag off pins to build functionality.
)

(event EventOnWeaponChanged
  (params (NewWeapon :type Fort%20Weapon%20Object%20Reference) (OldWeapon :type Fort%20Weapon%20Object%20Reference)))

(event EventOnWeaponFired)

(event EventOnNewWeapons
  (params (WeaponsAsItems :type Array%20of%20Fort%20Item%20Object%20References)))

(event EventOnCurieStateAttached
  (params (Tag :type Gameplay%20Tag%20Structure%20%28by%20ref%29)))

(event EventOnCurieStateDetached
  (params (Tag :type Gameplay%20Tag%20Structure%20%28by%20ref%29)))

(event EventOnBeginSkydiving
  (|CallParentFunction) ; This node is disabled and will not be called.\nDrag off pins to build functionality.
)

(event EventOnEndSkydiving
  (|CallParentFunction) ; This node is disabled and will not be called.\nDrag off pins to build functionality.
)

(event EventOnEnterVehicle
  (params (VehicleEntered :type Fort%20Vehicle%20Interface%20Interface%20%28by%20ref%29))
  (|CallParentFunction VehicleEntered) ; This node is disabled and will not be called.\nDrag off pins to build functionality.
)

(event EventOnExitVehicle
  (|CallParentFunction) ; This node is disabled and will not be called.\nDrag off pins to build functionality.
)

(event EventOnTargetingOrFirstPersonCameraChanged)

(event EventOnHitPawn
  (params (SelfActor :type Actor%20Object%20Reference) (OtherActor :type Actor%20Object%20Reference) (NormalImpulse :type Vector :default "0, 0, 0") (Hit :type Hit%20Result%20Structure%20%28by%20ref%29)))

(fn CalculateCurrentReactiveIntensity
  (return (* (Variables|Default|GetCurrentIntensity) (- 1.0 (Math|Float|Clamp(Float) (/ (- (Utilities|Time|GetGameTimeinSeconds) (Variables|Default|GetLastHitTime)) (Variables|Default|GetReactiveDecayRate)))))))

(fn ConstructionScript
  (|CallParentFunction))
