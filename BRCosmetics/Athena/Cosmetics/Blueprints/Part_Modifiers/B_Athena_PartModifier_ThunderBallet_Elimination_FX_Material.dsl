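; Cosmetic part modifier for the ThunderBallet character part that plays a material and Niagara reaction driven by kill count, apparently on eliminations.
; On each reactivity event it runs the parent logic, then writes game time and kill count to the owning part mesh's material parameters.
; It passes the part mesh's skin weight profile state to the idle Niagara VFX, restarts it flagged as reactive, then clears the flag and deactivates it after a short delay.
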(event EventOnReactivity
  (params (KillCount :type Integer))
  (bind _returnvalue (CustomCharacter|GetSkeletalMeshforPartonOwner))
  (|CallParentFunction KillCount)
  (Rendering|Material|SetScalarParameterValueonMaterials _returnvalue "Timestamp" (Utilities|Time|GetGameTimeinSeconds))
  (Rendering|Material|SetScalarParameterValueonMaterials _returnvalue "KillCount" (Math|Conversions|ToFloat(Integer) KillCount))
  (bind _asniagara_particle_system_component (Utilities|Casting|CastToNiagaraComponent (CallFunction|GetIdleVFXComponent)))
  (Niagara|SetNiagaraVariable(Bool) _asniagara_particle_system_component "SkinWeightCheck" (Components|SkinnedMesh|IsUsingSkinWeightProfile (Customization|GetSkeletalMeshforPartType (CustomCharacter|GetAssociatedPlayerPawn) (CustomCharacter|GetCharacterPartType) :!class /Script/FortniteGame.FortPlayerPawn)))
  (Niagara|ResetSystem _asniagara_particle_system_component)
  (Components|Activation|SetActive _asniagara_particle_system_component true true)
  (Niagara|SetNiagaraVariable(Float) _asniagara_particle_system_component "IsReactive" 1.0)
  (Utilities|FlowControl|Delay 4.0)
  (Niagara|SetNiagaraVariable(Float) _asniagara_particle_system_component "IsReactive" 0.0)
  (Components|Activation|Deactivate _asniagara_particle_system_component))

(fn ConstructionScript
  (|CallParentFunction))
