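; Cosmetic part modifier for BirdNest (ToonFE) that makes the part's idle particle effect react to weapon fire.
; Firing anything but a pickaxe sets the "RecentlyFired" particle flag, which a retriggerable delay later clears before reactivating the effect.
; After res-in it runs the toon front-end LUT replace.
; In the front end, a preview timeline inversely drives two mesh "Hide Element" material scalars and triggers a timed VFX burst.
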(event EventOnWeaponFired
  (bind _returnvalue (CallFunction|GetIdleVFXComponent))
  (bind (_asfort_weapon_pickaxe_athena _bsuccess) (Utilities|Casting|CastToFortWeaponPickaxeAthena (Weapon|GetCurrentWeapon (CustomCharacter|GetAssociatedPlayerPawn))))
  (|CallParentFunction)
  (if (not _bsuccess)
    (Effects|Components|ParticleSystem|SetBoolParameter _returnvalue "RecentlyFired" true)
    (Utilities|FlowControl|RetriggerableDelay 3.0)
    (Utilities|FlowControl|DoOnce false
      (:Completed
        (Effects|Components|ParticleSystem|SetBoolParameter _returnvalue "RecentlyFired" false)
        (CallFunction|ResetFired)))))

(event Custom|ResetFired
  (bind _returnvalue (CallFunction|GetIdleVFXComponent))
  (Components|Activation|Activate _returnvalue true)
  (Utilities|FlowControl|DoOnce false
    (:Completed
      (Effects|Components|ParticleSystem|SetBoolParameter _returnvalue "RecentlyFired" false)
      (CallFunction|ResetFired))))

(event EventOnPostResIn
  (bind _returnvalue (CustomCharacter|GetSkeletalMeshForPartOnPlayerPawn))
  (bind _returnvalue_2 (CallFunction|GetIdleVFXComponent))
  (|CallParentFunction)
  (Utilities|ToonFrontendLUTReplace
    (:Out
      ;[1 1 1] FE Preview
      (if (CustomCosmeticModifier|IsinFrontEnd)
        (bind (_direction _reactivity) (Timeline|FE_Preview PlayFromStart)
          (:Update
            (Rendering|Material|SetScalarParameterValueonMaterials _returnvalue "Hide Element 02" _reactivity)
            (Rendering|Material|SetScalarParameterValueonMaterials _returnvalue "Hide Element 05" (- 1.0 _reactivity)))
          (:FX%20Burst%20Event%20Track
            (Effects|Components|ParticleSystem|SetBoolParameter _returnvalue_2 "RecentlyFired" true)
            (Utilities|FlowControl|Delay 4.0)
            (Effects|Components|ParticleSystem|SetBoolParameter _returnvalue_2 "RecentlyFired" false)
            (CallFunction|FEFXReset)))))))

(event Custom|FE%20FX%20Reset
  (Components|Activation|Activate (CallFunction|GetIdleVFXComponent) true))

(fn ConstructionScript
  (|CallParentFunction))
