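; Cosmetic part modifier for the FastCheetah toon character part that tunes its outline in the front end.
; On post res-in, it calls the parent, then runs the ToonFrontendLUTReplace macro.
; After a short delay, if in the front end, it overrides outline fade distance and near and far thickness on the part mesh's overlay material.
; The construction script only calls its parent.
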(event EventOnPostResIn
  (bind _returnvalue (CustomCharacter|GetSkeletalMeshForPartOnPlayerPawn))
  (|CallParentFunction)
  (Utilities|ToonFrontendLUTReplace
    (:Out
      (Utilities|FlowControl|Delay 0.1)
      (if (CustomCosmeticModifier|IsinFrontEnd)
        (Class|FortCosmeticMaterialBlueprintLibrary|SetScalarParameterValueonOverlayMaterial _returnvalue "Outline_FadeDistance" 8e+02)
        (Class|FortCosmeticMaterialBlueprintLibrary|SetScalarParameterValueonOverlayMaterial _returnvalue "OutlineThickness_Near" 1.6)
        (Class|FortCosmeticMaterialBlueprintLibrary|SetScalarParameterValueonOverlayMaterial _returnvalue "OutlineThickness_Far" 0.5)))))

(fn ConstructionScript
  (|CallParentFunction))
