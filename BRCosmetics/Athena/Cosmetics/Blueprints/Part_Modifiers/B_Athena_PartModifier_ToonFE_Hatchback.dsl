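; Cosmetic part modifier that tunes the toon hatching look of a character part's materials in the front end.
; After res-in it calls the parent, runs the ToonFrontendLUTReplace utility, then checks for the front end.
; There it sets hatching line, NdotL and line-transition scalars on the part's skeletal mesh materials.
; Platforms other than mobile and PC performance mode also get a shading-curve camera exponent.
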(event EventOnPostResIn
  (bind _returnvalue (CustomCharacter|GetSkeletalMeshForPartOnPlayerPawn))
  (|CallParentFunction)
  (Utilities|ToonFrontendLUTReplace
    (:Out
      (if (CustomCosmeticModifier|IsinFrontEnd)
        ;[1 1 1] Hatching FE Adjustments
        (if (or (Fortnite|IsMobilePlatform) (Class|PlatformHelpersLibrary|IsPcPerfMode))
          (Rendering|Material|SetScalarParameterValueonMaterials _returnvalue "Hatching_Line_Amount" 2.5e+02)
          (Rendering|Material|SetScalarParameterValueonMaterials _returnvalue "Hatching_NdotL_Max" 0.9)
          (Rendering|Material|SetScalarParameterValueonMaterials _returnvalue "Hatching_NdotL_Min" 0.1)
          (Rendering|Material|SetScalarParameterValueonMaterials _returnvalue "LineTransition_Amount" 0.4)
          (else
            (Rendering|Material|SetScalarParameterValueonMaterials _returnvalue "Hatching_Line_Amount" 2.5e+02)
            (Rendering|Material|SetScalarParameterValueonMaterials _returnvalue "Hatching_NdotL_Max" 0.9)
            (Rendering|Material|SetScalarParameterValueonMaterials _returnvalue "Hatching_NdotL_Min" 0.1)
            (Rendering|Material|SetScalarParameterValueonMaterials _returnvalue "LineTransition_Amount" 0.4)
            (Rendering|Material|SetScalarParameterValueonMaterials _returnvalue "ShadingCurve_CameraExponent" 3.5)))))))

(fn ConstructionScript
  (|CallParentFunction))
