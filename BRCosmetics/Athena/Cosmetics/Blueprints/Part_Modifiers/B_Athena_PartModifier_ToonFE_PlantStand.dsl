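; Cosmetic part modifier for the PlantStand character that shows an eye-glow effect during its matching emote.
; On emote start, it compares persistent names and spawns the attached Niagara glow only for default-variant PlantStand or DoubleDuty.
; A timeline drives the effect's showFX parameter, and the component is destroyed on finish or cancellation.
; Cancellation uses an EmoteCancelled flag, plus an emote-playing check outside the front end.
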(event EventOnPostResIn
  (|CallParentFunction) ; This node is disabled and will not be called.\nDrag off pins to build functionality.
)

(event EventOnDayPhaseChanged
  (params (CurrentDayPhase :type EFortDayPhase%20Enum :default "Morning") (PreviousDayPhase :type EFortDayPhase%20Enum :default "Morning") (bAtCreation :type Boolean)))

(event EventOnNewPlayerWeaponType
  (params (NewType :type EFortItemType%20Enum :default "WorldItem") (OldType :type EFortItemType%20Enum :default "WorldItem"))
  (|CallParentFunction NewType OldType) ; This node is disabled and will not be called.\nDrag off pins to build functionality.
)

(event EventOnWeaponChanged
  (params (NewWeapon :type Fort%20Weapon%20Object%20Reference) (OldWeapon :type Fort%20Weapon%20Object%20Reference)))

(event EventOnWeaponFired)

(event EventOnNewWeapons
  (params (WeaponsAsItems :type Array%20of%20Fort%20Item%20Object%20References)))

(event EventOnCurieStateAttached
  (params (Tag :type Gameplay%20Tag%20Structure%20%28by%20ref%29)))

(event EventOnCurieStateDetached
  (params (Tag :type Gameplay%20Tag%20Structure%20%28by%20ref%29)))

(event EventOnBeginSkydiving
  (|CallParentFunction) ; This node is disabled and will not be called.\nDrag off pins to build functionality.
)

(event EventOnEndSkydiving
  (|CallParentFunction) ; This node is disabled and will not be called.\nDrag off pins to build functionality.
)

(event EventOnEnterVehicle
  (params (VehicleEntered :type Fort%20Vehicle%20Interface%20Interface%20%28by%20ref%29))
  (|CallParentFunction VehicleEntered) ; This node is disabled and will not be called.\nDrag off pins to build functionality.
)

(event EventOnExitVehicle
  (|CallParentFunction) ; This node is disabled and will not be called.\nDrag off pins to build functionality.
)

(event EventOnTargetingOrFirstPersonCameraChanged)

(event EventOnHitPawn
  (params (SelfActor :type Actor%20Object%20Reference) (OtherActor :type Actor%20Object%20Reference) (NormalImpulse :type Vector :default "0, 0, 0") (Hit :type Hit%20Result%20Structure%20%28by%20ref%29)))

(event EventOnBeginEmote
  (params (MontageItemDef :type Fort%20Item%20Definition%20Object%20Reference) (PawnEmoting :type Fort%20Pawn%20Object%20Reference))
  (bind _returnvalue (CosmeticLoadout|GetCosmeticLoadout))
  (bind _returnvalue_2 (CosmeticLoadout|GetItemDefinitionfromLoadout _returnvalue "(TagName=\"LoadoutSlot.Character\")"))
  (bind _returnvalue_3 (Item|GetPersistentName _returnvalue_2))
  (|CallParentFunction MontageItemDef PawnEmoting)
  (Variables|Default|SetEmoteCancelled false)
  (if (Utilities|String|Equal,CaseInsensitive(String) (Item|GetPersistentName MontageItemDef) (Variables|Default|GetPlantStandEmote_PersistentName))
    (bind _asfort_player_pawn (Utilities|Casting|CastToFortPlayerPawn PawnEmoting))
    (if (Utilities|String|Equal,CaseInsensitive(String) _returnvalue_3 (Variables|Default|GetPlantStandCharacter_PersistentName))
      (bind _index (CosmeticLoadout|GetVariantChannelActiveIndexforSlotActor "(TagName=\"LoadoutSlot.Character\")" _asfort_player_pawn "(TagName=\"Cosmetics.Variant.Channel.Progressive\")" (Utilities|Array|MakeArray "(TagName=\"Cosmetics.Variant.Property.Stage1\")" "(TagName=\"Cosmetics.Variant.Property.Stage2\")")))
      (bind _output_get (Variables|Default|SetVariant _index))
      ;[0.205 1 0.9303] Only spawns the Eye Glow FX if the character is PlantStand (Default Var) or DoubleDuty
      (if (== _output_get 0)
        _)
      (elif (Utilities|String|Equal,CaseInsensitive(String) (Variables|Default|GetDoubleDutyCharacter_PersistentName) _returnvalue_3)
        _))
    (bind _returnvalue_4 (PartModifier|SpawnSystem|SpawnSystemAttachedPartModifier "/BRCosmetics/Characters/Player/Female/Medium/Bodies/F_MED_PlantStand_Iris/FX/NS_PlantStand_Iris_EyeGlowIdle.NS_PlantStand_Iris_EyeGlowIdle" (CustomCharacter|GetSkeletalMeshForPartOnPlayerPawn) "None" "0, 0, 0" "0, 0, 0" "KeepRelativeOffset" false)
      (:Success
        (bind (_direction _eyeglowalpha) (Timeline|Timeline PlayFromStart)
          (:Update
            (if (CustomCosmeticModifier|IsinFrontEnd)
              (if (Variables|Default|GetEmoteCancelled)
                (Niagara|SetNiagaraVariableByString(Float) _returnvalue_4 "showFX" 0.0 :!label _label)
                (Variables|Default|SetEmoteCancelled true)
                (Components|DestroyComponent _returnvalue_4 :!label _label_2)
                (else
                  _))
              (elif (Emote|IsPlayingEmote (Fortnite|GetFortPlayerControllerfromActor PawnEmoting))
                _
                (elif (Variables|Default|GetEmoteCancelled)
                  _
                  (else
                    (goto _label)))))
            (Niagara|SetNiagaraVariableByString(Float) _returnvalue_4 "showFX" _eyeglowalpha))
          (:Finished
            _))
        (goto _label_2)))
    (else
      (Variables|Default|SetEmoteCancelled true))))

(fn ConstructionScript
  (|CallParentFunction))
