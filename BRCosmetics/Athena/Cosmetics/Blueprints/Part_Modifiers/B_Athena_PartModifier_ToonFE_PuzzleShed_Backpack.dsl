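; Backpack part modifier for the ToonFE PuzzleShed cosmetic that keeps strap shading on the player's body in sync with the backpack's material variant.
; When character customization finishes, it gets the Body skeletal mesh's dynamic material and the active index of the backpack slot's Material variant channel.
; That index switches the gradient-on-straps scalar parameter off for Mat0 and on for Mat1.
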(event EventOnPostResIn
  (|CallParentFunction))

(event EventOnCharacterCustomizationFinished
  (params (CharacterPartInfo :type Array%20of%20Applied%20Character%20Part%20Info%20Structures))
  (bind _returnvalue (Customization|GetSkeletalMeshforPartType (CustomCharacter|GetAssociatedPlayerPawn) "Body" :!class /Script/FortniteGame.FortPlayerPawn))
  (bind _material_instance_dynamic (Class|BTechArtGlobalLibrary|GetDynamicMaterialInstance _returnvalue))
  (bind _returnvalue_2 (CustomCharacter|GetAssociatedPlayerPawn))
  ; FORT-880779 Update shading of backpack straps on the body
  (|CallParentFunction CharacterPartInfo)
  (bind _asplayer_pawn_athena (Utilities|Casting|CastToPlayerPawn_Athena _returnvalue_2))
  (bind _index (CosmeticLoadout|GetVariantChannelActiveIndexforSlotActor "(TagName=\"LoadoutSlot.Backpack\")" _asplayer_pawn_athena "(TagName=\"Cosmetics.Variant.Channel.Material\")" (Utilities|Array|MakeArray "(TagName=\"Cosmetics.Variant.Property.Mat0\")" "(TagName=\"Cosmetics.Variant.Property.Mat1\")") false false))
  (switch int _index
    (:0
      (Rendering|Material|SetScalarParameterValue _material_instance_dynamic "PS_UseGradientOnBackpackStraps" 0.0 :!class /Script/Engine.MaterialInstanceDynamic))
    (:1
      (Rendering|Material|SetScalarParameterValue _material_instance_dynamic "PS_UseGradientOnBackpackStraps" 1.0 :!class /Script/Engine.MaterialInstanceDynamic))))

(fn ConstructionScript
  (|CallParentFunction))
