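; Part modifier for the TourBus outfit (CID_605) that makes the character part visually react to cosmetic stage and stat changes.
; On stage activation, the first and fourth Parts variants run a delayed timeline that lerps growth and glow material parameters from previous to current stage values.
; MaterialControl also snaps two extend effects to per-stage left/right sockets, activating them unless the head is hidden.
; Any stat change temporarily boosts idle particles and plays a kill-glow material timeline.
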(event EventOnCosmeticStateActivated
  (params (Stat%20Name :type Name :default "None") (Stat%20Value :type Integer) (STAGE%20Num%20Being%20Activated :type Integer))
  (|CallParentFunction Stat%20Name Stat%20Value STAGE%20Num%20Being%20Activated)
  (bind _isreactivevariant (CallFunction|CheckVariant))
  (if _isreactivevariant
    (Development|PrintString (Utilities|String|Append "Stage Num Being Activated:  " (Utilities|String|ToString(Integer) STAGE%20Num%20Being%20Activated)))
    (Utilities|FlowControl|Delay 0.2)
    (bind (_direction _emissive___vfx_lerp _opacity_lerp _noisespreadcurve _emissiveglowstrength) (Timeline|State%20Change PlayFromStart)
      (:Update
        (CallFunction|MaterialControl STAGE%20Num%20Being%20Activated _emissive___vfx_lerp _opacity_lerp _noisespreadcurve _emissiveglowstrength)))))

(event EventOnStatChanged
  (params (State%20Name :type Name :default "None") (Stat%20Value :type Integer))
  (bind _returnvalue (CallFunction|GetIdleVFXComponent))
  (|CallParentFunction State%20Name Stat%20Value)
  (Effects|Components|ParticleSystem|SetFloatParameter _returnvalue "SpawnRate" 1e+01)
  (Effects|Components|ParticleSystem|SetFloatParameter _returnvalue "Alpha" 2.0)
  (bind (_direction _killglowcontrol) (Timeline|KillGlowTimeline PlayFromStart)
    (:Update
      (Rendering|Material|SetScalarParameterValueonMaterials (CustomCharacter|GetSkeletalMeshForPartOnPlayerPawn) "KillGlowControl" _killglowcontrol))
    (:Finished
      (Effects|Components|ParticleSystem|SetFloatParameter _returnvalue "SpawnRate" 1.0)
      (Effects|Components|ParticleSystem|SetFloatParameter _returnvalue "Alpha" 0.5))))

(fn CheckVariant
  (bind _index (Class|FortCosmeticFunctionLibrary|GetVariantChannelActiveIndexForItem "/BRCosmetics/Athena/Items/Cosmetics/Characters/CID_605_Athena_Commando_M_TourBus.CID_605_Athena_Commando_M_TourBus" (CosmeticLoadout|GetCosmeticLoadout) "(TagName=\"Cosmetics.Variant.Channel.Parts\")" (Utilities|Array|MakeArray "(TagName=\"Cosmetics.Variant.Property.Stage1\")" "(TagName=\"Cosmetics.Variant.Property.Stage2\")" "(TagName=\"Cosmetics.Variant.Property.Stage3\")" "(TagName=\"Cosmetics.Variant.Property.Stage4\")")))
  (return (or (== _index 0) (== _index 3))))

(fn MaterialControl
  (params (Stage :type Integer) (Emissive%20Timeline%20Control :type Float%20%28double-precision%29) (Opacity%20Timeline%20Control :type Float%20%28double-precision%29) (NoiseFadeCurve :type Float%20%28double-precision%29) (EmissiveGlowStrengthCurve :type Float%20%28double-precision%29))
  (bind _returnvalue (CustomCharacter|GetSkeletalMeshForPartOnPlayerPawn))
  (bind _returnvalue_2 (CustomCharacter|GetSkeletalMeshForPartOnPlayerPawn))
  (bind _returnvalue_3 (- Stage 1))
  (bind _returnvalue_4 (> _returnvalue_3 0))
  (bind _returnvalue_5 (select _returnvalue_4 _returnvalue_3 0))
  (bind _ns_tourbus_extend_b (Variables|Default|GetNS_TourBus_Extend_B))
  (bind _ns_tourbus_extend_a (Variables|Default|GetNS_TourBus_Extend_A))
  (Utilities|FlowControl|Sequence
    (:then_0
      (Transformation|SetWorldTransform _ns_tourbus_extend_a (Transformation|GetSocketTransform _returnvalue (Utilities|Array|Get(aref) (Variables|Default|GetFX_BoneArrayLeft) Stage) "RTS_World") false false)
      (if (not (Variables|HiddenPartCheck|GetbIsHeadHidden))
        (Components|Activation|Activate _ns_tourbus_extend_a false)))
    (:then_1
      (Transformation|SetWorldTransform _ns_tourbus_extend_b (Transformation|GetSocketTransform _returnvalue (Utilities|Array|Get(aref) (Variables|Default|GetFX_BoneArrayRight) Stage) "RTS_World") false false)
      (if (not (Variables|HiddenPartCheck|GetbIsHeadHidden))
        (Components|Activation|Activate _ns_tourbus_extend_b false)))
    (:then_2
      (Rendering|Material|SetScalarParameterValueonMaterials _returnvalue_2 "EmisiveGrowth" (Math|Float|Lerp (Utilities|Array|Get(acopy) (Variables|Default|GetMaterial_EmissiveGrowthValues) (select (> (- Stage 1) 0) (- Stage 1) 0)) (Utilities|Array|Get(acopy) (Variables|Default|GetMaterial_EmissiveGrowthValues) Stage) Emissive%20Timeline%20Control))
      (Rendering|Material|SetScalarParameterValueonMaterials _returnvalue_2 "BaseGrowth" (Math|Float|Lerp (Utilities|Array|Get(acopy) (Variables|Default|GetMaterial_BaseGrowthValues) _returnvalue_5) (Utilities|Array|Get(acopy) (Variables|Default|GetMaterial_BaseGrowthValues) Stage) Opacity%20Timeline%20Control))
      (Rendering|Material|SetScalarParameterValueonMaterials _returnvalue_2 "NoiseSpread" NoiseFadeCurve)
      (Rendering|Material|SetScalarParameterValueonMaterials _returnvalue_2 "OnFireGlowControl" (Math|Float|Lerp (Utilities|Array|Get(acopy) (Variables|Default|GetMaterial_EmissiveGlowStrength) _returnvalue_5) (Utilities|Array|Get(acopy) (Variables|Default|GetMaterial_EmissiveGlowStrength) Stage) EmissiveGlowStrengthCurve)))))

(fn ConstructionScript
  (|CallParentFunction))
