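; Cosmetic part modifier for the Traffic Hat that adjusts its glass material when shown in the front end.
; On the post-res-in event it gets the part's skeletal mesh on the player pawn and runs the parent logic.
; In the front end, it scans the mesh materials for names containing "Glass_PBR" and enables the UseDistanceRoughness parameter on their dynamic material instances.
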(event EventOnPostResIn
  (bind _returnvalue (CustomCharacter|GetSkeletalMeshForPartOnPlayerPawn))
  (|CallParentFunction)
  (if (CustomCosmeticModifier|IsinFrontEnd)
    (for (_array_element _array_index) (Rendering|Material|GetMaterials _returnvalue)
      (if (Utilities|String|Contains (Utilities|GetDisplayName _array_element) "Glass_PBR")
        (Rendering|Material|SetScalarParameterValue (Class|BTechArtGlobalLibrary|GetDynamicMaterialInstance _returnvalue _array_index) "UseDistanceRoughness" 1.0 :!class /Script/Engine.MaterialInstanceDynamic)))))

(fn ConstructionScript
  (|CallParentFunction))
