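; Character part modifier for the TreasureHunterFashionMint back bling pack that toggles the pickaxe in its idle VFX.
; On weapon change (after a short delay) and skydive end, it queries pickaxe meta tags through the part's anim instance, setting ShowPickaxe false on a TreasureHunterFashionMint match, otherwise true.
; Skydiving forces the pickaxe visible after a delay.
; OnBegin and IsSkyDiving flags gate the weapon-change check; post-res-in clears both.
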(event EventOnWeaponChanged
  (params (NewWeapon :type Fort%20Weapon%20Object%20Reference) (OldWeapon :type Fort%20Weapon%20Object%20Reference))
  (bind _returnvalue (CallFunction|GetIdleVFXComponent))
  (|CallParentFunction NewWeapon OldWeapon)
  (Utilities|FlowControl|Delay 0.1)
  (if (Variables|Default|GetOnBegin)
    _
    (elif (Variables|Default|GetIsSkyDiving)
      _
      (else
        (bind _ascustom_character_part_anim_instance (Utilities|Casting|CastToCustomCharacterPartAnimInstance (Components|SkeletalMesh|GetAnimInstance (Customization|GetSkeletalMeshforPartType (CustomCharacter|GetAssociatedPlayerPawn) (CustomCharacter|GetCharacterPartType) :!class /Script/FortniteGame.FortPlayerPawn))))
        (if (CosmeticTags|PickaxeItemDefinitionHasAnyMetaTag _ascustom_character_part_anim_instance (Utilities|Array|MakeArray "(TagName=\"Cosmetics.MetaTags.TreasureHunterFashionMint\")") true)
          (Effects|Components|ParticleSystem|SetBoolParameter _returnvalue "ShowPickaxe" false)
          (else
            (Effects|Components|ParticleSystem|SetBoolParameter _returnvalue "ShowPickaxe" true)))))))

(event EventOnPostResIn
  (|CallParentFunction)
  (Variables|Default|SetOnBegin false)
  (Variables|Default|SetIsSkyDiving false))

(event EventOnBeginSkydiving
  (|CallParentFunction)
  (Utilities|FlowControl|Delay 0.1)
  (Variables|Default|SetIsSkyDiving true)
  (Effects|Components|ParticleSystem|SetBoolParameter (CallFunction|GetIdleVFXComponent) "ShowPickaxe" true))

(event EventOnEndSkydiving
  (bind _returnvalue (CallFunction|GetIdleVFXComponent))
  (|CallParentFunction)
  (Variables|Default|SetIsSkyDiving false)
  (bind _ascustom_character_part_anim_instance (Utilities|Casting|CastToCustomCharacterPartAnimInstance (Components|SkeletalMesh|GetAnimInstance (Customization|GetSkeletalMeshforPartType (CustomCharacter|GetAssociatedPlayerPawn) (CustomCharacter|GetCharacterPartType) :!class /Script/FortniteGame.FortPlayerPawn))))
  (if (CosmeticTags|PickaxeItemDefinitionHasAnyMetaTag _ascustom_character_part_anim_instance (Utilities|Array|MakeArray "(TagName=\"Cosmetics.MetaTags.TreasureHunterFashionMint\")") true)
    (Effects|Components|ParticleSystem|SetBoolParameter _returnvalue "ShowPickaxe" false)
    (else
      (Effects|Components|ParticleSystem|SetBoolParameter _returnvalue "ShowPickaxe" true))))

(fn ConstructionScript
  (|CallParentFunction)
  (Variables|Default|SetOnBegin false))
