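; Cosmetic body part modifier for the TripleScoopFlammin character that adds effects to both feet.
; After res-in it attaches left and right foot components, snapped, to the FX_foot_L and FX_foot_R sockets on the pawn part's skeletal mesh.
; On weapon change it runs a hidden-part check on both components, keyed to the HideFX.Feet animation tag, presumably hiding them when that tag applies.
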(event EventOnPostResIn
  (bind _returnvalue (CustomCharacter|GetSkeletalMeshForPartOnPlayerPawn))
  (|CallParentFunction)
  (Transformation|AttachComponentToComponent (Variables|Default|GetP_TripleScoopFlammin_Foot_L) _returnvalue "FX_foot_L" "SnapToTarget" "SnapToTarget" "SnapToTarget" true)
  (Transformation|AttachComponentToComponent (Variables|Default|GetP_TripleScoopFlammin_Foot_R) _returnvalue "FX_foot_R" "SnapToTarget" "SnapToTarget" "SnapToTarget" true))

(event EventOnWeaponChanged
  (params (NewWeapon :type Fort%20Weapon%20Object%20Reference) (OldWeapon :type Fort%20Weapon%20Object%20Reference))
  (|CallParentFunction NewWeapon OldWeapon)
  (Default|HiddenPartCheck-Weapon (Utilities|Array|MakeArray (Variables|Default|GetP_TripleScoopFlammin_Foot_R) (Variables|Default|GetP_TripleScoopFlammin_Foot_L)) NewWeapon OldWeapon false "(TagName=\"Animation.CharacterParts.HideFX.Feet\")" "bisFootHidden"))

(fn ConstructionScript)
