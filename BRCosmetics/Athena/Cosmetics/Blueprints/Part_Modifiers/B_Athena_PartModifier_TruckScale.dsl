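; Cosmetic part modifier for the TruckScale character part that plays a damage-reactive material and Niagara effect on the player.
; Rising DamageReceived stat stamps game time into material parameters and spawns attached FX, with a timer that stamps the fade-out and separate handling when downed (DBNO).
; The front end plays a delayed preview instead.
; Hidden hand/foot checks drive material and particle parameters, and the material variant index feeds the FX.
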(event EventOnCosmeticPartAppliedPostResIn
  (params (Part :type Custom%20Character%20Part%20Object%20Reference))
  (bind _returnvalue (CustomCharacter|GetSkeletalMeshForPartOnPlayerPawn))
  ;[1 0.3744 0.1648] Hidden hand/feet/head check - Tech Art
  (|CallParentFunction Part)
  (if (Variables|HiddenPartCheck|GetbIsHandHidden)
    (Rendering|Material|SetScalarParameterValueonMaterials _returnvalue "HideArmFX" 1.0)
    (else
      (Rendering|Material|SetScalarParameterValueonMaterials _returnvalue "HideArmFX" 0.0)))
  (if (Variables|HiddenPartCheck|GetbIsFootHidden)
    (Rendering|Material|SetScalarParameterValueonMaterials _returnvalue "HideLegFX" 1.0)
    (else
      (Rendering|Material|SetScalarParameterValueonMaterials _returnvalue "HideLegFX" 0.0))))

(event EventOnWeaponChanged
  (params (NewWeapon :type Fort%20Weapon%20Object%20Reference) (OldWeapon :type Fort%20Weapon%20Object%20Reference))
  (bind _returnvalue (CustomCharacter|GetSkeletalMeshForPartOnPlayerPawn))
  ;[1 0.3744 0.1648] Hidden hand/feet/head check - Tech Art
  (|CallParentFunction NewWeapon OldWeapon)
  (if (Variables|HiddenPartCheck|GetbIsHandHidden)
    (Rendering|Material|SetScalarParameterValueonMaterials _returnvalue "HideArmFX" 1.0)
    (else
      (Rendering|Material|SetScalarParameterValueonMaterials _returnvalue "HideArmFX" 0.0)))
  (if (Variables|HiddenPartCheck|GetbIsFootHidden)
    (Rendering|Material|SetScalarParameterValueonMaterials _returnvalue "HideLegFX" 1.0)
    (else
      (Rendering|Material|SetScalarParameterValueonMaterials _returnvalue "HideLegFX" 0.0))))

(event EventOnPostResIn
  (bind _returnvalue (CustomCharacter|GetSkeletalMeshForPartOnPlayerPawn))
  (bind _returnvalue_2 (Utilities|Time|GetGameTimeinSeconds))
  (bind _attachedns (Variables|Default|GetAttachedNS))
  (|CallParentFunction)
  (if (CustomCosmeticModifier|IsinFrontEnd)
    ;[0.5417 0.03734 0 0.5] FrontEnd
    (CallFunction|CheckCurrentVariant)
    (Utilities|FlowControl|Delay 3.0)
    (Rendering|Material|SetScalarParameterValueonMaterials (CustomCharacter|GetSkeletalMeshForPartOnPlayerPawn) "InputTimeStampB" 0.0)
    (Rendering|Material|SetScalarParameterValueonMaterials (CustomCharacter|GetSkeletalMeshForPartOnPlayerPawn) "InputTimeStamp" (Utilities|Time|GetGameTimeinSeconds))
    (CallFunction|SpawnReactivityFX)
    (Utilities|FlowControl|Delay 4.0)
    (Rendering|Material|SetScalarParameterValueonMaterials (CustomCharacter|GetSkeletalMeshForPartOnPlayerPawn) "InputTimeStampB" (Utilities|Time|GetGameTimeinSeconds))
    (Utilities|FlowControl|RetriggerableDelay 4.5)
    (else
      ;[0.15 0.03457 0.01581 0.5] Reactivity
      (Utilities|FlowControl|Sequence
        (:then_0
          (bind (_statname _statvalue) (Cosmetics|WatchCosmeticStat (CustomCharacter|GetAssociatedPlayerPawn) "(TagName=\"StatManager.DamageReceived\")" 0 false)
            (:OnStatChanged
              (CallFunction|CheckCurrentVariant)
              (if (stale Custom|OnDBNOStateChanged bInIsDBNO)
                (Utilities|FlowControl|RetriggerableDelay 4.0)
                (Rendering|Material|SetScalarParameterValueonMaterials _returnvalue "InputTimeStamp" _returnvalue_2)
                (Components|Activation|Deactivate _attachedns)
                (Variables|Default|SetReceivingDamage false)
                (Variables|Default|SetDamageRecieved 0)
                (elif (> _statvalue (Variables|Default|GetDamageRecieved))
                  (if (Variables|Default|GetReceivingDamage)
                    (Utilities|Time|ClearandInvalidateTimerbyHandle (stale _returnvalue_3))
                    (Components|Activation|Deactivate _attachedns)
                    (CallFunction|SpawnReactivityFX)
                    (else
                      (Rendering|Material|SetScalarParameterValueonMaterials _returnvalue "InputTimeStampB" 0.0)
                      (Rendering|Material|SetScalarParameterValueonMaterials _returnvalue "InputTimeStamp" _returnvalue_2)
                      (CallFunction|SpawnReactivityFX)
                      (Variables|Default|SetReceivingDamage true)))
                  (bind _returnvalue_3 (Utilities|Time|SetTimerbyEvent (delegate Custom|FadeOut) 4.0)))))))
        (:then_1
          (EventDispatchers|BindEventtoOnDBNOStateChanged (CustomCharacter|GetAssociatedPlayerPawn) (delegate Custom|OnDBNOStateChanged)))))))

(event Custom|FadeOut
  ;[0.15 0.03457 0.01581 0.5] Reactivity
  (Rendering|Material|SetScalarParameterValueonMaterials (CustomCharacter|GetSkeletalMeshForPartOnPlayerPawn) "InputTimeStampB" (Utilities|Time|GetGameTimeinSeconds))
  (Variables|Default|SetReceivingDamage false))

(event Custom|OnDBNOStateChanged
  (params (FortPawn :type Fort%20Pawn%20Object%20Reference) (bInIsDBNO :type Boolean)))

(fn CheckCurrentVariant
  (bind _index (CosmeticLoadout|GetVariantChannelActiveIndexforSlotActor "(TagName=\"LoadoutSlot.Character\")" (CustomCharacter|GetAssociatedPlayerPawn) "(TagName=\"Cosmetics.Variant.Channel.Material\")" (Utilities|Array|MakeArray "(TagName=\"Cosmetics.Variant.Property.Mat1\")" "(TagName=\"Cosmetics.Variant.Property.Mat2\")")))
  (Variables|Default|SetCurrentVariant _index)
  (return))

(fn SpawnReactivityFX
  (bind _attachedns (Variables|Default|GetAttachedNS))
  (bind _returnvalue (PartModifier|SpawnSystem|SpawnSystemAttachedPartModifier (Variables|Default|GetReactivityNS) (CustomCharacter|GetSkeletalMeshForPartOnPlayerPawn))
    (:Success
      (Variables|Default|SetAttachedNS _returnvalue)
      (Effects|Components|ParticleSystem|SetBoolParameter _attachedns "bisHandHidden" (Variables|HiddenPartCheck|GetbIsHandHidden))
      (Effects|Components|ParticleSystem|SetBoolParameter _attachedns "bisFootHidden" (Variables|HiddenPartCheck|GetbIsFootHidden))
      (Effects|Components|ParticleSystem|SetFloatParameter _attachedns "Variant" (Math|Conversions|ToFloat(Integer) (Variables|Default|GetCurrentVariant)))
      (Utilities|Array|Add (Variables|HiddenPartCheck|GetAdditional%20FX%20Needing%20Part%20Checks) (Variables|Default|GetAttachedNS)))
    (:Failure
      _))
  (return))

(fn ConstructionScript
  (|CallParentFunction))
