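; Part modifier Blueprint for the TurboFueled backpack that plays a Niagara burst effect in reaction to player kills.
; After res-in, it checks whether frontend reactivity should display; if so, it spawns a preview burst on the part mesh after a short delay.
; Otherwise it watches the AthenaKills cosmetic stat on the associated pawn and spawns the burst attached at the mesh's Root whenever that stat changes.
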(event EventOnPostResIn
  (bind _returnvalue (CustomCharacter|GetSkeletalMeshForPartOnPlayerPawn))
  (|CallParentFunction)
  (if (Reactive|Frontend|ShouldDisplayFrontendReactivity 0 "Backpack")
    (if (Reactive|Frontend|ShouldDisplayFrontendReactivity _returnvalue "Backpack")
      (Utilities|FlowControl|Delay 2.0)
      (Niagara|SpawnSystemAttached "/BRCosmetics/Accessories/FORT_Backpacks/Backpack_M_MED_TurboFueled/FX/NS_Backpack_TurboFueled_Burst.NS_Backpack_TurboFueled_Burst" _returnvalue "Root")
      (Utilities|FlowControl|Delay 5.0))
    (else
      (Cosmetics|WatchCosmeticStat (CustomCharacter|GetAssociatedPlayerPawn) "(TagName=\"StatManager.AthenaKills\")" 0 false
        (:OnStatChanged
          (Niagara|SpawnSystemAttached "/BRCosmetics/Accessories/FORT_Backpacks/Backpack_M_MED_TurboFueled/FX/NS_Backpack_TurboFueled_Burst.NS_Backpack_TurboFueled_Burst" (CustomCharacter|GetSkeletalMeshForPartOnPlayerPawn) "Root"))))))

(fn ConstructionScript
  (|CallParentFunction))
