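; Cosmetic part modifier for the TweakSmirk character that plays an elimination-tagged effect and material reaction when a tracked stat changes to a positive value.
; Each spawns attached VFX, then OnReactivity writes game time to Timestamp_0 or Timestamp_1 by count parity on the part's dynamic material.
; In the front end, post-init sets Threshold_Exp on head, body and face, then previews the reaction after a short delay if allowed.
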(event EventOnPostResIn
  (|CallParentFunction)
  (if (CustomCosmeticModifier|IsinFrontEnd)
    (Rendering|Material|SetScalarParameterValueonMaterials (Utilities|Array|MakeArray (Customization|GetSkeletalMeshforPartType (CustomCharacter|GetAssociatedPlayerPawn) "Head" :!class /Script/FortniteGame.FortPlayerPawn) (Customization|GetSkeletalMeshforPartType (CustomCharacter|GetAssociatedPlayerPawn) "Body" :!class /Script/FortniteGame.FortPlayerPawn) (Customization|GetSkeletalMeshforPartType (CustomCharacter|GetAssociatedPlayerPawn) "Face" :!class /Script/FortniteGame.FortPlayerPawn)) "Threshold_Exp" 1.5)
    (Utilities|FlowControl|Delay 1.0)
    (if (CallFunction|ShouldDisplayFrontendReactivity)
      (bind _returnvalue (CallFunction|GetVFXSpawnParameters "(TagName=\"CosmeticPart.VFX.Elimination\")")
        (:Success
          (PartModifier|SpawnSystem|SpawnSystemAttachedPartModifier (Utilities|Struct|BreakCharacterPartModifierVFXSpawnParameters _returnvalue) (.AttachToComponent (Utilities|Struct|BreakCharacterPartModifierVFXSpawnParameters _returnvalue)) (.SocketName (Utilities|Struct|BreakCharacterPartModifierVFXSpawnParameters _returnvalue)) "0, 0, 0" "0, 0, 0" (.LocationType (Utilities|Struct|BreakCharacterPartModifierVFXSpawnParameters _returnvalue)) true (.bAutoActivate (Utilities|Struct|BreakCharacterPartModifierVFXSpawnParameters _returnvalue))
            (:Success
              (CallFunction|OnReactivity 1)
              (Utilities|FlowControl|Delay 1e+01))))))))

(event EventOnStatChanged
  (params (State%20Name :type Name :default "None") (Stat%20Value :type Integer))
  (|CallParentFunction State%20Name Stat%20Value)
  (if (> Stat%20Value 0)
    (bind _returnvalue (CallFunction|GetVFXSpawnParameters "(TagName=\"CosmeticPart.VFX.Elimination\")")
      (:Success
        (PartModifier|SpawnSystem|SpawnSystemAttachedPartModifier (Utilities|Struct|BreakCharacterPartModifierVFXSpawnParameters _returnvalue) (.AttachToComponent (Utilities|Struct|BreakCharacterPartModifierVFXSpawnParameters _returnvalue)) (.SocketName (Utilities|Struct|BreakCharacterPartModifierVFXSpawnParameters _returnvalue)) "0, 0, 0" "0, 0, 0" (.LocationType (Utilities|Struct|BreakCharacterPartModifierVFXSpawnParameters _returnvalue)) true (.bAutoActivate (Utilities|Struct|BreakCharacterPartModifierVFXSpawnParameters _returnvalue))
          (:Success
            (CallFunction|OnReactivity Stat%20Value)))))))

(event Custom|OnReactivity
  (params (KillCount :type Integer))
  (Rendering|Material|SetScalarParameterValue (Class|BTechArtGlobalLibrary|GetDynamicMaterialInstance (CustomCharacter|GetSkeletalMeshForPartOnPlayerPawn)) (Utilities|String|StringToName (Utilities|String|Append "Timestamp_" (Utilities|String|ToString(Integer) (Math|Integer|%25%28Integer%29 KillCount 2)))) (Utilities|Time|GetGameTimeinSeconds) :!class /Script/Engine.MaterialInstanceDynamic))

(fn ConstructionScript
  (|CallParentFunction))
