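; Cosmetic part modifier for Ultralight that makes character and backpack materials and idle Niagara VFX react to damage dealt.
; When the Parts variant is reactive, it gathers or creates dynamic material instances for backpack or body, face and head.
; Damage since a baseline drives emissive and animation timing; a retriggerable delay later resets that baseline.
; In the frontend, a looping timer simulates rising damage with pauses between cycles.
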(event EventOnPostResIn
  (bind _returnvalue (CustomCharacter|GetAssociatedPlayerPawn))
  (|CallParentFunction)
  (Utilities|FlowControl|Sequence
    (:then_0
      (Default|CheckifReactive)
      ;[1 1 1] Materials Setup
      (if (Variables|Default|GetbReactive)
        (if (Utilities|Enum|Equal(Enum) (CustomCharacter|GetCharacterPartType) "Backpack")
          (CallFunction|CreateMIDsForPart (Customization|GetSkeletalMeshforPartType _returnvalue "Backpack" :!class /Script/FortniteGame.FortPlayerPawn))
          (else
            (CallFunction|CreateMIDsForPart (Customization|GetSkeletalMeshforPartType _returnvalue "Body" :!class /Script/FortniteGame.FortPlayerPawn))
            (CallFunction|CreateMIDsForPart (Customization|GetSkeletalMeshforPartType _returnvalue "Face" :!class /Script/FortniteGame.FortPlayerPawn))
            (CallFunction|CreateMIDsForPart (Customization|GetSkeletalMeshforPartType _returnvalue "Head" :!class /Script/FortniteGame.FortPlayerPawn))))))
    (:then_1
      (bind (_statname _statvalue) (Cosmetics|WatchCosmeticStat (CustomCharacter|GetAssociatedPlayerPawn) "(TagName=\"StatManager.DamageDealt\")")
        (:OnStatChanged
          (CallFunction|Reactivity _statvalue)
          (Utilities|FlowControl|RetriggerableDelay (Variables|Default|GetDamageResetDuration))
          (Variables|Default|SetDamage%20Reset _statvalue))))
    (:then_2
      ;[1 1 1] Frontend simulate reactivity
      (if (and (Reactive|Frontend|IsCosmeticObjectDisplayedinFrontendContext 0) (Variables|Default|GetbReactive))
        (Utilities|FlowControl|Delay 2.5)
        (bind _returnvalue_2 (Utilities|Time|SetTimerbyEvent (delegate Custom|FE%20Reactivity) 0.066 true))
        (Variables|Default|SetFE_TimerEventHandle _returnvalue_2)))))

(event Custom|FE%20Reactivity
  (bind _fe_timereventhandle (Variables|Default|GetFE_TimerEventHandle))
  ;[1 1 1] Frontend simulate reactivity
  (if (< (Variables|Default|GetFE_DamageSim) 200)
    (bind _output_get (Variables|Default|SetFE_DamageSim (+ (Variables|Default|GetFE_DamageSim) 5)))
    (CallFunction|Reactivity _output_get)
    (else
      (Utilities|Time|PauseTimerbyHandle _fe_timereventhandle)
      (Variables|Default|SetFE_DamageSim 0)
      (Utilities|FlowControl|Delay 8.0)
      (Utilities|Time|UnpauseTimerbyHandle _fe_timereventhandle))))

(event Custom|Reactivity
  (params (Damage :type Integer))
  (bind _damage_reset (Variables|Default|GetDamage%20Reset))
  (bind _returnvalue (- Damage _damage_reset))
  (bind _returnvalue_2 (Math|Conversions|ToFloat(Integer) _returnvalue))
  (bind _returnvalue_3 (/ _returnvalue_2 2e+02))
  (bind _returnvalue_4 (Math|Float|Clamp(Float) _returnvalue_3))
  (bind _returnvalue_5 (Math|Float|Lerp 0.0 3.0 _returnvalue_4))
  ;[1 1 1] Reactivity
  (if (> Damage 0)
    (Utilities|FlowControl|Sequence
      (:then_0
        (for _array_element (Variables|Default|GetMIDs)
          (Rendering|Material|SetScalarParameterValue _array_element "RE_Anim1Time_CCPM" (+ _returnvalue_5 (Utilities|Time|GetGameTimeinSeconds)) :!class /Script/Engine.MaterialInstanceDynamic)
          (Rendering|Material|SetScalarParameterValue _array_element "RE_EmissiveIntensityMultiplier" (Math|Float|Lerp 0.5 2.0 _returnvalue_4) :!class /Script/Engine.MaterialInstanceDynamic)
          (Rendering|Material|SetScalarParameterValue _array_element "RE_Anim2Time_CCPM" (Utilities|Time|GetGameTimeinSeconds) :!class /Script/Engine.MaterialInstanceDynamic)
          (Variables|Default|SetDamageResetDuration (+ 1.0 _returnvalue_5)))
        (Niagara|SetNiagaraVariable(Float) (Utilities|Casting|CastToNiagaraComponent (CallFunction|GetIdleVFXComponent)) "HitTime" (+ (Math|Float|Lerp 0.0 4.2 (Math|Float|Power _returnvalue_4 0.5)) (Utilities|Time|GetRealTimeSeconds)))))))

(fn CheckIfReactive
  (bind _returnvalue (CustomCharacter|GetAssociatedPlayerPawn))
  (if (Utilities|Enum|Equal(Enum) (CustomCharacter|GetCharacterPartType) "Backpack")
    ;[1 1 1] Backpack
    (bind _index (CosmeticLoadout|GetVariantChannelActiveIndexforSlotActor "(TagName=\"LoadoutSlot.Backpack\")" _returnvalue "(TagName=\"Cosmetics.Variant.Channel.Parts\")" (Utilities|Array|MakeArray "(TagName=\"Cosmetics.Variant.Property.Stage1\")" "(TagName=\"Cosmetics.Variant.Property.Stage2\")")))
    (switch int _index
      (:0
        (Variables|Default|SetbReactive true))
      (:1
        (Variables|Default|SetbReactive false)))
    (else
      (bind _index_2 (CosmeticLoadout|GetVariantChannelActiveIndexforSlotActor "(TagName=\"LoadoutSlot.Character\")" _returnvalue "(TagName=\"Cosmetics.Variant.Channel.Parts\")" (Utilities|Array|MakeArray "(TagName=\"Cosmetics.Variant.Property.Stage1\")" "(TagName=\"Cosmetics.Variant.Property.Stage2\")")))
      (switch int _index_2
        (:0
          (Variables|Default|SetbReactive true))
        (:1
          (Variables|Default|SetbReactive false))))))

(fn CreateMIDsforPart
  (params (Skeletal%20Mesh%20Component :type Skeletal%20Mesh%20Component%20Object%20Reference))
  (Utilities|IsValid Skeletal%20Mesh%20Component
    (:Is%20Valid
      (for _array_element (Rendering|Material|GetMaterials Skeletal%20Mesh%20Component)
        (bind _asmaterial_instance_dynamic (Utilities|Casting|CastToMaterialInstanceDynamic _array_element)
          (:then
            (Utilities|Array|Add (Variables|Default|GetMIDs) _asmaterial_instance_dynamic))
          (:CastFailed
            (bind _returnvalue (Rendering|Material|CreateDynamicMaterialInstance _array_element "None" "None" :!class /Script/Engine.KismetMaterialLibrary))
            (Utilities|Array|Add (Variables|Default|GetMIDs) _returnvalue)))))))

(fn ConstructionScript
  (|CallParentFunction))
