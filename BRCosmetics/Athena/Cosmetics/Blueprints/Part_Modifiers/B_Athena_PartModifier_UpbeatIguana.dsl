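; Cosmetic part modifier blueprint for the UpbeatIguana character part that turns on custom depth rendering for its mesh.
; On the EventOnPostResIn event it calls the parent, then fetches the part's skeletal mesh on the player pawn and enables render custom depth on it.
; The construction script only calls the parent.
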(event EventOnPostResIn
  (|CallParentFunction)
  (Rendering|SetRenderCustomDepth (CustomCharacter|GetSkeletalMeshForPartOnPlayerPawn) true))

(fn ConstructionScript
  (|CallParentFunction))
