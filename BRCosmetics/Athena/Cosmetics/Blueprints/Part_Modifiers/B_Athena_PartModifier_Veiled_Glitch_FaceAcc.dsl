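; Face accessory part modifier for the Veiled Glitch character cosmetic that adds idle and reactive glitch effects.
; On post-res-in, it attaches the idle VFX component to the Root of the part's skeletal mesh.
; It watches chest-open and kill stats, spawning the same Niagara head burst after a short delay when either changes.
; The kill reaction requires a positive count and also plays a sound at the mesh when the burst spawns successfully.
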(event EventOnPostResIn
  (bind _returnvalue (CustomCharacter|GetSkeletalMeshForPartOnPlayerPawn))
  (|CallParentFunction)
  (Utilities|FlowControl|Sequence
    (:then_0
      ;[1 1 1] Attach Idle VFX
      (Transformation|AttachComponentToComponent (CallFunction|GetIdleVFXComponent) (CustomCharacter|GetSkeletalMeshForPartOnPlayerPawn) "Root" "SnapToTarget" "SnapToTarget" "KeepRelative" true))
    (:then_1
      ;[1 0 0.3426] Chest Opening Reactivity
      (Cosmetics|WatchCosmeticStat (CustomCharacter|GetAssociatedPlayerPawn) "(TagName=\"StatManager.OpenChests\")" 0 false
        (:OnStatChanged
          (Utilities|FlowControl|Delay 0.5)
          (PartModifier|SpawnSystem|SpawnSystemAttachedPartModifier "/Game/Characters/Player/Female/Medium/Bodies/F_MED_Veiled_Glitch/FX/NS_Veiled_Glitch_Character_Head_Burst.NS_Veiled_Glitch_Character_Head_Burst" (CustomCharacter|GetSkeletalMeshForPartOnPlayerPawn) "Root" "0, 0, 0" "0, 0, 0" "KeepWorldPosition" true true false))))
    (:then_2
      ;[0 0.7366 1] Kill Reactivity
      (bind (_statname _statvalue) (Cosmetics|WatchCosmeticStat (CustomCharacter|GetAssociatedPlayerPawn) "(TagName=\"StatManager.AthenaKills\")" 0 false)
        (:OnStatChanged
          (Utilities|FlowControl|Delay 0.5)
          (if (> _statvalue 0)
            (PartModifier|SpawnSystem|SpawnSystemAttachedPartModifier "/Game/Characters/Player/Female/Medium/Bodies/F_MED_Veiled_Glitch/FX/NS_Veiled_Glitch_Character_Head_Burst.NS_Veiled_Glitch_Character_Head_Burst" _returnvalue "Root" "0, 0, 0" "0, 0, 0" "KeepWorldPosition" true true false
              (:Success
                (Audio|SpawnSoundatLocation (Variables|Default|GetKillReaction_Veiled_SFX) (Transformation|GetSocketLocation _returnvalue "None"))))))))))

(fn ConstructionScript
  (|CallParentFunction))
