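; Body part modifier for the VelvetDeskFiber cosmetic that makes the character react visually and audibly as the player gains eliminations.
; In matches, a watch on the AthenaKills stat keeps the highest clamped stage, drives material parameters, triggers sounds, and at stage three adds hand-fire burst FX.
; In the front end, a timeline previews three stages instead.
; Idle VFX bool parameters select the effect level, and a game-time material stamp marks each stat change.
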(event EventOnPawnAndPartChanged
  (|CallParentFunction)
  (if (not (CustomCosmeticModifier|IsinFrontEnd))
    (bind _returnvalue (Class|FortCosmeticFunctionLibrary|GetCosmeticStatValue (CustomCharacter|GetAssociatedPlayerPawn) "(TagName=\"StatManager.AthenaKills\")"))
    (bind _stage (CallFunction|StatValuetoStage _returnvalue))
    (Variables|Default|SetCurrentStage _stage)
    (CallFunction|UpdateMaterialParameters)))

(event EventOnCosmeticStateActivated
  (params (Stat%20Name :type Name :default "None") (Stat%20Value :type Integer) (STAGE%20Num%20Being%20Activated :type Integer))
  (|CallParentFunction Stat%20Name Stat%20Value STAGE%20Num%20Being%20Activated)
  (Variables|Default|SetCurrentStage STAGE%20Num%20Being%20Activated)
  (CallFunction|UpdateMaterialParameters))

(event EventPreDestruction
  (|CallParentFunction)
  (Utilities|Time|ClearandInvalidateTimerbyHandle (Variables|Default|GetRefreshTimeHandler)))

(event EventOnPostResIn
  (bind _returnvalue (CustomCharacter|GetSkeletalMeshForPartOnPlayerPawn))
  (bind _returnvalue_2 (CustomCharacter|GetSkeletalMeshForPartOnPlayerPawn))
  (bind _endstage (Variables|Default|GetEndStage))
  (bind _returnvalue_3 (+ _endstage 1))
  (bind _lastvalue (Variables|Default|GetLastValue))
  (bind _animdurationperstage (Variables|Default|GetAnimDurationPerStage))
  (bind _lastvalue_2 (Variables|Default|GetLastValue))
  (|CallParentFunction)
  (Utilities|FlowControl|Sequence
    (:then_1
      (if (CustomCosmeticModifier|IsinFrontEnd)
        ; In FE mode: Its gonna be 3 stage represents Normal > Emissive light up > FX up
        (Timeline|Timeline_0 PlayFromStart
          (:Stage%201
            ;; FX
            (CallFunction|StageOne)
            (Rendering|Material|SetScalarParameterValueonMaterials _returnvalue_2 "Animation_Duration" (Utilities|Array|Get(aref) _animdurationperstage))
            (Rendering|Material|SetScalarParameterValueonMaterials _returnvalue_2 "CurrentStage" 0.0))
          (:Stage%202
            (CallFunction|StageTwo)
            (Rendering|Material|SetScalarParameterValueonMaterials _returnvalue_2 "Animation_Duration" (Utilities|Array|Get(aref) _animdurationperstage 1))
            (Rendering|Material|SetScalarParameterValueonMaterials _returnvalue_2 "CurrentStage" 1.0))
          (:Stage%203
            (CallFunction|StageThree)
            (PartModifier|SpawnSystem|SpawnSystemAttachedPartModifier "/BRCosmetics/Characters/Player/Male/Medium/Bodies/M_MED_VelvetDeskFiber/FX/NS_Character_VelvetDeskFiber_HandFire_Burst.NS_Character_VelvetDeskFiber_HandFire_Burst" (CustomCharacter|GetSkeletalMeshForPartOnPlayerPawn) "FX_Hand"
              (:Success
                (Rendering|Material|SetScalarParameterValueonMaterials _returnvalue_2 "Animation_Duration" (Utilities|Array|Get(aref) _animdurationperstage 3))
                (Rendering|Material|SetScalarParameterValueonMaterials _returnvalue_2 "CurrentStage" 3.0)))))
        (CallFunction|OnActualStatChange)
        (else
          ; In non FE mode: Its gonna be 5 stages represents 5 eliminates
          (bind (_statname _statvalue) (Cosmetics|WatchCosmeticStat (CustomCharacter|GetAssociatedPlayerPawn) "(TagName=\"StatManager.AthenaKills\")")
            (:OnStatChanged
              (Variables|Default|SetLastValue (Math|Integer|SelectInt (Math|Integer|Max(Integer) _statvalue _lastvalue) _endstage (< (Math|Integer|Max(Integer) _statvalue _lastvalue) _returnvalue_3)))
              (Utilities|FlowControl|Sequence
                (:then_0
                  (CallFunction|InitializeVariabels)
                  (Rendering|Material|SetScalarParameterValueonMaterials _returnvalue "Animation_Duration" (Utilities|Array|Get(aref) (Variables|Default|GetAnimDurationPerStage) _lastvalue_2))
                  (Rendering|Material|SetScalarParameterValueonMaterials _returnvalue "CurrentStage" (Math|Conversions|ToFloat(Integer) _lastvalue_2))
                  (if (== _lastvalue_2 3)
                    (CallFunction|StageThree)
                    (CallFunction|PlayReactiveAudio "(TagName=\"SoundLibrary.Cosmetic.Other.One\")")
                    (PartModifier|SpawnSystem|SpawnSystemAttachedPartModifier "/BRCosmetics/Characters/Player/Male/Medium/Bodies/M_MED_VelvetDeskFiber/FX/NS_Character_VelvetDeskFiber_HandFire_Burst.NS_Character_VelvetDeskFiber_HandFire_Burst" (CustomCharacter|GetSkeletalMeshForPartOnPlayerPawn) "FX_Hand")
                    (elif (== (Variables|Default|GetLastValue) 1)
                      (CallFunction|PlayReactiveAudio "(TagName=\"SoundLibrary.Cosmetic.Other.Two\")"))))
                (:then_1
                  (if (and (> (Math|Integer|Max(Integer) _statvalue _lastvalue) (Variables|Default|GetStartStage)) (< (Math|Integer|Max(Integer) _statvalue _lastvalue) _returnvalue_3))
                    (CallFunction|OnActualStatChange)))))))))))

(event Custom|HackyWorkaroundRefreshTimerGuy
  (bind _returnvalue (CustomCharacter|GetSkeletalMeshForPartOnPlayerPawn))
  (bind _lastvalue (Variables|Default|GetLastValue))
  ; In non FE mode: Its gonna be 5 stages represents 5 eliminates
  (Utilities|FlowControl|Sequence
    (:then_0
      (CallFunction|InitializeVariabels)
      (Rendering|Material|SetScalarParameterValueonMaterials _returnvalue "Animation_Duration" (Utilities|Array|Get(aref) (Variables|Default|GetAnimDurationPerStage) _lastvalue))
      (Rendering|Material|SetScalarParameterValueonMaterials _returnvalue "CurrentStage" (Math|Conversions|ToFloat(Integer) _lastvalue))
      (if (== _lastvalue 3)
        (CallFunction|StageThree)
        (CallFunction|PlayReactiveAudio "(TagName=\"SoundLibrary.Cosmetic.Other.One\")")
        (PartModifier|SpawnSystem|SpawnSystemAttachedPartModifier "/BRCosmetics/Characters/Player/Male/Medium/Bodies/M_MED_VelvetDeskFiber/FX/NS_Character_VelvetDeskFiber_HandFire_Burst.NS_Character_VelvetDeskFiber_HandFire_Burst" (CustomCharacter|GetSkeletalMeshForPartOnPlayerPawn) "FX_Hand")
        (elif (== (Variables|Default|GetLastValue) 1)
          (CallFunction|PlayReactiveAudio "(TagName=\"SoundLibrary.Cosmetic.Other.Two\")"))))))

(event Custom|OnActualStatChange
  ; In non FE mode: Its gonna be 5 stages represents 5 eliminates
  ;;[1 1 1] This should only fire off when the stat changes - use this event!
  (Rendering|Material|SetScalarParameterValueonMaterials (CustomCharacter|GetSkeletalMeshForPartOnPlayerPawn) "GameTimeFromBP" (Utilities|Time|GetGameTimeinSeconds)))

(event Custom|StageOne
  (bind _returnvalue (CallFunction|GetIdleVFXComponent))
  ; FX Stage
  (Effects|Components|ParticleSystem|SetBoolParameter _returnvalue "Stage_01" false)
  (Effects|Components|ParticleSystem|SetBoolParameter _returnvalue "Stage_02" false))

(event Custom|PlayReactiveAudio
  (params (Event%20Name :type Gameplay%20Tag%20Structure))
  ; Reactive Audio
  (Networking|CanExecuteCosmeticEvents
    (:True
      (SoundLibrary|PlaySoundSimple (WorldSubsystems|GetSoundLibrarySubsystem) self (Utilities|Struct|MakeSimpleSoundSettings Event%20Name true false true true 1.0 1.0 0.0 0 0 "None" (CustomCharacter|GetSkeletalMeshForPartOnPlayerPawn)) :!class /Script/SoundLibrary.SoundLibrarySubsystem))))

(event Custom|InitializeVariabels
  (bind _returnvalue (CustomCharacter|GetSkeletalMeshForPartOnPlayerPawn))
  ; Initialize Variables
  (Rendering|Material|SetScalarParameterValueonMaterials _returnvalue "StartAtStage" (Math|Conversions|ToFloat(Integer) (Variables|Default|GetStartStage)))
  (Rendering|Material|SetScalarParameterValueonMaterials _returnvalue "StopAtStage" (Math|Conversions|ToFloat(Integer) (Variables|Default|GetEndStage))))

(event Custom|StageTwo
  (bind _returnvalue (CallFunction|GetIdleVFXComponent))
  ; FX Stage
  (Effects|Components|ParticleSystem|SetBoolParameter _returnvalue "Stage_01" true)
  (Effects|Components|ParticleSystem|SetBoolParameter _returnvalue "Stage_02" false))

(event Custom|StageThree
  (bind _returnvalue (CallFunction|GetIdleVFXComponent))
  ; FX Stage
  (Effects|Components|ParticleSystem|SetBoolParameter _returnvalue "Stage_01" true)
  (Effects|Components|ParticleSystem|SetBoolParameter _returnvalue "Stage_02" true))

(fn StatValueToStage
  (params (StatValue :type Integer))
  (bind (_arrayindex _arrayelement) (Utilities|Array|ReverseforEachLoop (Variables|PartModifierControls|WatchCosmeticStat|GetStat%20Thresholds))
    (:LoopBody
      (if (>= (Variables|GetStatValue) _arrayelement)
        (return _arrayindex)))
    (:Completed
      (return 0))))

(fn UpdateMaterialParameters
  (switch int (Variables|Default|GetCurrentStage)
    (:0
      (CallFunction|StageOne))
    (:1
      (CallFunction|StageTwo))))

(fn ConstructionScript
  (bind _returnvalue (CustomCharacter|GetSkeletalMeshForPartOnPlayerPawn))
  (|CallParentFunction)
  (Rendering|Material|SetScalarParameterValueonMaterials _returnvalue "GameTimeFromBP" -99999.0)
  (Rendering|Material|SetScalarParameterValueonMaterials _returnvalue "StartAtStage" (Math|Conversions|ToFloat(Integer) (Variables|Default|GetStartStage)))
  (Rendering|Material|SetScalarParameterValueonMaterials _returnvalue "StopAtStage" (Math|Conversions|ToFloat(Integer) (Variables|Default|GetEndStage))))
