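; Cosmetic part modifier for the WeaveHarbor character that plays a reactive head effect when a tracked stat changes to a positive value.
; TriggerReactivity spawns the reactivity particle system attached to the head, passes it the head-hidden state, and plays the OnKill sound outside the front end.
; In the front end, the effect also fires once shortly after res-in.
; Targeting or camera changes update the idle VFX's bIsAiming parameter, which stays false in the front end.
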(event EventOnPostResIn
  (|CallParentFunction)
  (Utilities|FlowControl|Delay 1.5)
  (if (CustomCosmeticModifier|IsinFrontEnd)
    (CallFunction|TriggerReactivity)
    (Utilities|FlowControl|Delay 6.0)))

(event EventOnTargetingOrFirstPersonCameraChanged
  (|CallParentFunction)
  (if (and (Variables|/Script/FortniteGame.FortPlayerPawn|GetbIsTargeting (CustomCharacter|GetAssociatedPlayerPawn)) (not (CustomCosmeticModifier|IsinFrontEnd)))
    (Effects|Components|ParticleSystem|SetBoolParameter (CallFunction|GetIdleVFXComponent) "bIsAiming" true)
    (else
      (Effects|Components|ParticleSystem|SetBoolParameter (CallFunction|GetIdleVFXComponent) "bIsAiming" false))))

(event EventOnStatChanged
  (params (State%20Name :type Name :default "None") (Stat%20Value :type Integer))
  (|CallParentFunction State%20Name Stat%20Value)
  (if (> Stat%20Value 0)
    (CallFunction|TriggerReactivity)))

(event Custom|TriggerReactivity
  (bind _returnvalue (CustomCharacter|GetSkeletalMeshForPartOnPlayerPawn))
  (bind _returnvalue_2 (PartModifier|SpawnSystem|SpawnSystemAttachedPartModifier (Variables|Default|GetReactivityNS) _returnvalue "head" "0, 0, 0" "0, 0, 0" "SnapToTarget" true true true "AutoRelease")
    (:Success
      (Effects|Components|ParticleSystem|SetBoolParameter _returnvalue_2 "bisHeadHidden" (Variables|HiddenPartCheck|GetbIsHeadHidden))
      ;[1 1 1] Check to avoid playing the sound on FE
      (if (not (CustomCosmeticModifier|IsinFrontEnd))
        (Audio|SpawnSoundAttached "/BRCosmetics/Sounds/Characters/WeaveHarbor/MSS_Character_WeaveHarbor_Reactive_OnKill.MSS_Character_WeaveHarbor_Reactive_OnKill" _returnvalue)))))

(fn ConstructionScript
  (|CallParentFunction))
