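; Part modifier Blueprint for the WeaveHarbor cosmetic character that plays reactive VFX on one material variant.
; It triggers shortly after res-in in the front end, or when a stat changes to a positive value.
; CheckVariant reads the loadout's material channel to keep particles off the default variant.
; TriggerReactivity briefly flags the idle VFX and spawns an attached Niagara effect, passing it hidden hand and foot flags.
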(event EventOnPostResIn
  (|CallParentFunction)
  (Utilities|FlowControl|Delay 1.5)
  (if (CustomCosmeticModifier|IsinFrontEnd)
    (CallFunction|CheckVariant)
    (if (Variables|Default|GetIsVariant1)
      (CallFunction|TriggerReactivity)
      (Utilities|FlowControl|Delay 6.0))))

(event EventOnStatChanged
  (params (State%20Name :type Name :default "None") (Stat%20Value :type Integer))
  (|CallParentFunction State%20Name Stat%20Value)
  (if (> Stat%20Value 0)
    (CallFunction|CheckVariant)
    (if (Variables|Default|GetIsVariant1)
      (CallFunction|TriggerReactivity))))

(event Custom|CheckVariant
  ;[1 1 1] Loadout check to avoid particles spawning on Default Variant
  (bind _index (Class|FortCosmeticFunctionLibrary|GetVariantChannelActiveIndexForItem "/BRCosmetics/Athena/Items/Cosmetics/Characters/Character_WeaveHarbor.Character_WeaveHarbor" (CosmeticLoadout|GetCosmeticLoadout) "(TagName=\"Cosmetics.Variant.Channel.Material\")" (Utilities|Array|MakeArray "(TagName=\"Cosmetics.Variant.Property.Mat1\")" "(TagName=\"Cosmetics.Variant.Property.Mat2\")" "(TagName=\"Cosmetics.Variant.Property.Mat3\")")))
  (Variables|Default|SetIsVariant1 (== _index 1)))

(event Custom|TriggerReactivity
  (bind _returnvalue (CallFunction|GetIdleVFXComponent))
  (Effects|Components|ParticleSystem|SetBoolParameter _returnvalue "TriggerReactivity" true)
  (bind _returnvalue_2 (PartModifier|SpawnSystem|SpawnSystemAttachedPartModifier "/BRCosmetics/Characters/Player/Male/Medium/Bodies/M_MED_WeaveHarbor/FX/NS_WeaveHarbor_Body_Reactivity.NS_WeaveHarbor_Body_Reactivity" (CustomCharacter|GetSkeletalMeshForPartOnPlayerPawn) "root")
    (:Success
      (Effects|Components|ParticleSystem|SetBoolParameter _returnvalue_2 "bisFootHidden" (Variables|HiddenPartCheck|GetbIsFootHidden))
      (Effects|Components|ParticleSystem|SetBoolParameter _returnvalue_2 "bisHandHidden" (Variables|HiddenPartCheck|GetbIsHandHidden))
      (Utilities|FlowControl|Delay 2.0)
      (Effects|Components|ParticleSystem|SetBoolParameter _returnvalue "TriggerReactivity" false))))

(fn ConstructionScript
  (|CallParentFunction))
