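; Part modifier for the WidePlane backpack cosmetic, adding reactivity to chest opens and shield gains.
; In frontend or preview, it writes game time to the mesh's material parameter once after a delay, despite the event's "Loop" name.
; Otherwise it watches both stats, calling audio and chest hooks or a shield hook; ChestOpen and SheildIncrease are empty here.
; Helper functions gate an attached shield sound on quick-healing item tags, with no callers shown.
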(event Custom|FrontEndReactivityLoop
  ;[0.4766 0.4867 1] FrontEnd Reactivity Loop
  (Utilities|FlowControl|Delay 13.33)
  (Rendering|Material|SetScalarParameterValueonMaterials (CustomCharacter|GetSkeletalMeshForPartOnPlayerPawn) "BP_GetGameTime" (Utilities|Time|GetGameTimeinSeconds)))

(event EventOnPostResIn
  (bind _returnvalue (CustomCharacter|GetAssociatedPlayerPawn))
  (|CallParentFunction)
  (if (or (Reactive|Frontend|IsCosmeticObjectDisplayedinFrontendContext 0) (Reactive|Frontend|IsCosmeticObjectDisplayedasPreview 0))
    (CallFunction|FrontEndReactivityLoop)
    (else
      (Utilities|FlowControl|Sequence
        (:then_0
          (Cosmetics|WatchCosmeticStat _returnvalue "(TagName=\"StatManager.OpenChests\")" 0 false
            (:OnStatChanged
              (CallFunction|AudioStatWatcher "None" 1 (Variables|Default|GetAudioReactivityTag))
              (CallFunction|ChestOpen))))
        (:then_1
          (Cosmetics|WatchCosmeticStat _returnvalue "(TagName=\"StatManager.GainedShieldTimes\")" 0 false
            (:OnStatChanged
              (CallFunction|SheildIncrease))))))))

(event Custom|ChestOpen)

(event Custom|SheildIncrease)

(fn AllowShieldAudio
  (params (GameplayAbility :type Fort%20Gameplay%20Ability%20Object%20Reference) (AbilityTags :type Gameplay%20Tag%20Container%20Structure))
  (if (GameplayTags|MatchesAnyTags "(TagName=\"Gameplay.Action.Player.QuickHealingItem\")" AbilityTags true)
    (Variables|Default|SetbAllowReactiveAudio true)))

(fn PlayShieldReactiveAudio
  (if (Variables|Default|GetbAllowReactiveAudio)
    (Audio|SpawnSoundAttached (Variables|Default|GetShieldGainSoundCue) (CustomCharacter|GetSkeletalMeshForPartOnPlayerPawn))
    (Variables|Default|SetbAllowReactiveAudio false)))

(fn ConstructionScript)
