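; Character part modifier for the Winged Fury tail cosmetic that turns its idle VFX off when the player wears a back bling.
; After res-in, it calls the parent, then gets the part's anim instance from the pawn's skeletal mesh as a custom character part anim instance.
; A back bling check on that instance sets the wearing-backbling flag and, from its long-backbling result, the cape flag.
; Either flag deactivates the idle VFX component; otherwise it is activated.
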(event EventOnPostResIn
  (bind _returnvalue (CallFunction|GetIdleVFXComponent))
  ;[0.5117 0.5418 1] Check for BackBling
  (|CallParentFunction)
  (bind _ascustom_character_part_anim_instance (Utilities|Casting|CastToCustomCharacterPartAnimInstance (Components|SkeletalMesh|GetAnimInstance (Customization|GetSkeletalMeshforPartType (CustomCharacter|GetAssociatedPlayerPawn) (CustomCharacter|GetCharacterPartType) :!class /Script/FortniteGame.FortPlayerPawn)))
    (:then
      (Variables|Default|SetisWearingBackbling (CustomCharacter|CheckforBackbling _ascustom_character_part_anim_instance 5e+01))
      (Variables|Default|SetisWearingCape (.bIsBackblingLong (CustomCharacter|CheckforBackbling _ascustom_character_part_anim_instance 5e+01))))
    (:CastFailed
      _))
  ;[0.09549 1 0.4203] Turn off FX when wearing a backpack
  (if (or (Variables|Default|GetisWearingCape) (Variables|Default|GetisWearingBackbling))
    (Components|Activation|Deactivate _returnvalue)
    (else
      (Components|Activation|Activate _returnvalue false))))

(fn ConstructionScript)
