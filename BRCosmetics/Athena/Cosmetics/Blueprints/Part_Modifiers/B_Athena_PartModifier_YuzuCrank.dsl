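; YuzuCrank cosmetic part modifier that makes a character part react to eliminations, weapon changes and front-end preview.
; Reactivity writes backdatable game-time stamps into mesh materials and idle VFX, once per frame.
; A counter's parity alternates between two timestamp parameters, so successive pulses can overlap.
; On init and weapon change, it sets the VFX's ReactiveBool according to whether the held weapon uses the YuzuCrankTide pickaxe mesh.
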(event EventOnPostResIn
  (|CallParentFunction)
  (Utilities|FlowControl|Sequence
    (:then_0
      (if (or (CustomCosmeticModifier|IsinFrontEnd) (Reactive|Frontend|IsCosmeticObjectDisplayedinFrontendContext 0))
        ;[0 0.3774 0.7083 0.5] Frontend Preview Loop
        (Utilities|FlowControl|Delay 2.0)
        (CallFunction|Reactivity 0.0)
        (Utilities|FlowControl|Delay 1e+01)))
    (:then_1
      ; Register Reactivity On Kills
      (bind (_statname _statvalue) (Cosmetics|WatchCosmeticStat (CustomCharacter|GetAssociatedPlayerPawn) "(TagName=\"StatManager.AthenaKills\")" 0 false)
        (:OnStatChanged
          (if (> _statvalue 0)
            (CallFunction|Reactivity 0.0)))))
    (:then_2
      (bind _asfort_player_pawn_athena (Utilities|Casting|CastToFortPlayerPawnAthena (CustomCharacter|GetAssociatedPlayerPawn)))
      (CallFunction|CheckCurrentWeapon (Variables|/Script/FortniteGame.FortPawn|GetCurrentWeapon _asfort_player_pawn_athena)))))

(event EventOnWeaponChanged
  (params (NewWeapon :type Fort%20Weapon%20Object%20Reference) (OldWeapon :type Fort%20Weapon%20Object%20Reference))
  (|CallParentFunction NewWeapon OldWeapon)
  (CallFunction|Reactivity 5.0)
  (CallFunction|CheckCurrentWeapon NewWeapon))

(event Custom|Reactivity
  (params (ExtraDuration :type Float%20%28double-precision%29))
  (bind _elimn_reactivity_counter (Variables|Default|GetElimn%20Reactivity%20Counter))
  (bind _returnvalue (Math|Integer|%25%28Integer%29 _elimn_reactivity_counter 2))
  (bind _returnvalue_2 (Utilities|String|ToString(Integer) _returnvalue))
  (bind _returnvalue_3 (Utilities|String|Append "Timestamp_" _returnvalue_2))
  (bind _returnvalue_4 (Utilities|String|StringToName _returnvalue_3))
  (bind _returnvalue_5 (Utilities|Time|GetGameTimeinSeconds))
  (bind _returnvalue_6 (- _returnvalue_5 ExtraDuration))
  (Utilities|DoOncePerFrame
    (:%20%20
      (Math|Integer|IncrementInt _elimn_reactivity_counter
        (:%20%20
          (Rendering|Material|SetScalarParameterValueonMaterials (CustomCharacter|GetSkeletalMeshForPartOnPlayerPawn) _returnvalue_4 _returnvalue_6)
          (Effects|Components|ParticleSystem|SetFloatParameter (CallFunction|GetIdleVFXComponent) _returnvalue_4 _returnvalue_6))))))

(event Custom|CheckCurrentWeapon
  (params (Target :type Fort%20Weapon%20Object%20Reference))
  (Effects|Components|ParticleSystem|SetBoolParameter (CallFunction|GetIdleVFXComponent) "ReactiveBool" (== (Variables|/Script/Engine.SkinnedMeshComponent|GetSkeletalMesh (Visual|GetWeaponMesh Target)) "/BRCosmetics/Weapons/FORT_Melee/Pickaxe_YuzuCrankTide/Meshes/YuzuCrankTide_Axe.YuzuCrankTide_Axe")))

(fn ConstructionScript
  (|CallParentFunction))
