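; Cosmetic part modifier for ZirconSweep that makes the character's materials react progressively to eliminations when the matching Particle variant is active.
; In game it watches the kill stat; in the frontend it previews an elimination and rewinds via timeline at max.
; OnElimination sets start-time, progress and alternating pulse material parameters, blending from interpolated progress if retriggered mid-animation.
; It also pulses jump bubbles on landing and polls body LOD for the idle VFX.
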(event EventOnPawnLanded
  (bind _returnvalue (CallFunction|GetIdleVFXComponent))
  (|CallParentFunction)
  (Effects|Components|ParticleSystem|SetBoolParameter _returnvalue "bJumpBubbles" true)
  (Utilities|FlowControl|Delay 1.0)
  (Effects|Components|ParticleSystem|SetBoolParameter _returnvalue "bJumpBubbles" false))

(event Custom|RestartFrontendReactivityLoop
  ;[1 1 1] Frontend Reactivty Preview
  (Variables|Default|SetEliminations 0)
  (Utilities|FlowControl|Delay 1.0)
  (bind _result (Math|Integer|IncrementInt (Variables|Default|GetEliminations))
    (:%20%20
      (CallFunction|OnElimination _result)
      (Utilities|FlowControl|Delay 5.0)
      (if (>= (Variables|Default|GetEliminations) (Variables|Default|GetMaxEliminations))
        (bind (_direction _progress _masteramount) (Timeline|RewindFrontendReactivity PlayFromStart)
          (:Update
            (Rendering|Material|SetScalarParameterValueonMaterials (CustomCharacter|GetSkeletalMeshForPartOnPlayerPawn) "ReactivePreviousProgress_CCPM" _progress)
            (Rendering|Material|SetScalarParameterValueonMaterials (CustomCharacter|GetSkeletalMeshForPartOnPlayerPawn) "ReactiveProgress_CCPM" _progress)
            (Rendering|Material|SetScalarParameterValueonMaterials (CustomCharacter|GetSkeletalMeshForPartOnPlayerPawn) "ReactiveMasterAmount_CCPM" _masteramount))
          (:Finished
            (Rendering|Material|SetScalarParameterValueonMaterials (CustomCharacter|GetSkeletalMeshForPartOnPlayerPawn) "ReactiveStartTime_CCPM" -1e+02)
            (Rendering|Material|SetScalarParameterValueonMaterials (CustomCharacter|GetSkeletalMeshForPartOnPlayerPawn) "ElimPulseStartTime_CCPM" -1e+02)
            (Rendering|Material|SetScalarParameterValueonMaterials (CustomCharacter|GetSkeletalMeshForPartOnPlayerPawn) "ElimPulseStartTime2_CCPM" -1e+02)
            (CallFunction|RestartFrontendReactivityLoop)))
        (else
          _)))))

(event EventOnPostResIn
  ;[1 1 1] Frontend Reactivty Preview
  (|CallParentFunction)
  (Utilities|FlowControl|Sequence
    (:then_0
      (bind _index (Class|FortCosmeticFunctionLibrary|GetVariantChannelActiveIndexForSlot "(TagName=\"LoadoutSlot.Character\")" (CosmeticLoadout|GetCosmeticLoadout) "(TagName=\"Cosmetics.Variant.Channel.Particle\")" (Utilities|Array|MakeArray "(TagName=\"Cosmetics.Variant.Property.Particle1\")" "(TagName=\"Cosmetics.Variant.Property.Particle2\")")))
      (if (== _index 0)
        (if (CustomCosmeticModifier|IsinFrontEnd)
          (Variables|Default|SetEliminations 0)
          (Utilities|FlowControl|Delay 1.0)
          (bind _result (Math|Integer|IncrementInt (Variables|Default|GetEliminations))
            (:%20%20
              (CallFunction|OnElimination _result)
              (Utilities|FlowControl|Delay 5.0)
              (if (>= (Variables|Default|GetEliminations) (Variables|Default|GetMaxEliminations))
                (Timeline|RewindFrontendReactivity PlayFromStart)
                (else
                  _))))
          (else
            ;[1 1 1] Ingame init
            (Variables|Default|SetEliminations 0)
            (bind (_statname _statvalue) (Cosmetics|WatchCosmeticStat (CustomCharacter|GetAssociatedPlayerPawn) "(TagName=\"StatManager.AthenaKills\")" 0 false)
              (:OnStatChanged
                (CallFunction|OnElimination _statvalue)))))))
    (:then_1
      ;; LOD Check for Niagara
      (bind _ascustom_character_part_anim_instance (Utilities|Casting|CastToCustomCharacterPartAnimInstance (Components|SkeletalMesh|GetAnimInstance (Customization|GetSkeletalMeshforPartType (CustomCharacter|GetAssociatedPlayerPawn) "Body" :!class /Script/FortniteGame.FortPlayerPawn))))
      (Variables|Default|SetBodyAccAnimInstance _ascustom_character_part_anim_instance)
      (Utilities|Time|SetTimerbyEvent (EventDispatchers|CreateEvent :!event Custom|LODCheckBody) 2.0 true))))

(event Custom|LODCheckBody
  (bind _returnvalue (CallFunction|GetIdleVFXComponent))
  (if (Reactive|Frontend|IsCosmeticObjectDisplayedinFrontendContext 0)
    (Effects|Components|ParticleSystem|SetIntParameter _returnvalue "BodyAccLOD" 0)
    (else
      (Effects|Components|ParticleSystem|SetIntParameter _returnvalue "BodyAccLOD" (Variables|/Script/FortniteGame.CustomCharacterPartAnimInstance|GetCurrentLOD (Variables|Default|GetBodyAccAnimInstance))))))

(event Custom|OnElimination
  (params (NumElims :type Integer))
  ;[1 1 1] Prevent same-frame retriggering
  (if (not (Variables|Default|GetReactivityInProgress))
    (Variables|Default|SetEliminations NumElims)
    (Variables|Default|SetReactivityInProgress true)
    (if (not (Variables|Default|GetAnimInProgress))
      ;[1 1 1] Set progressive reactivity anim start & end values
      (Variables|Default|SetPreviousProgress (/ (Math|Conversions|ToFloat(Integer) (Math|Integer|Clamp(Integer) (- (Variables|Default|GetEliminations) 1) 0 (Variables|Default|GetMaxEliminations))) (Math|Conversions|ToFloat(Integer) (Variables|Default|GetMaxEliminations))))
      (else
        ;;[1 1 1] Handle subsequent elimination during active animation
        (Variables|Default|SetPreviousProgress (Math|Float|Lerp (Variables|Default|GetTargetProgress) (Variables|Default|GetPreviousProgress) (/ (Utilities|Time|GetTimerRemainingTimebyHandle (Variables|Default|GetReactivityAnimTimer)) (Variables|Default|GetReactivityAnimDuration))))
        (Utilities|Time|ClearandInvalidateTimerbyHandle (Variables|Default|GetReactivityAnimTimer))))
    (Variables|Default|SetTargetProgress (/ (Math|Conversions|ToFloat(Integer) (Variables|Default|GetEliminations)) (Math|Conversions|ToFloat(Integer) (Variables|Default|GetMaxEliminations))))
    ;[1 1 1] Update material parameters
    (Rendering|Material|SetScalarParameterValueonMaterials (CustomCharacter|GetSkeletalMeshForPartOnPlayerPawn) "ReactivePreviousProgress_CCPM" (Variables|Default|GetPreviousProgress))
    (Rendering|Material|SetScalarParameterValueonMaterials (CustomCharacter|GetSkeletalMeshForPartOnPlayerPawn) "ReactiveProgress_CCPM" (Variables|Default|GetTargetProgress))
    (Rendering|Material|SetScalarParameterValueonMaterials (CustomCharacter|GetSkeletalMeshForPartOnPlayerPawn) "ReactiveStartTime_CCPM" (Utilities|Time|GetGameTimeinSeconds))
    (Rendering|Material|SetScalarParameterValueonMaterials (CustomCharacter|GetSkeletalMeshForPartOnPlayerPawn) (Utilities|Select :Index (Math|Integer|%25%28Integer%29 (Variables|Default|GetEliminations) 2) :Option%200 "ElimPulseStartTime_CCPM" :Option%201 "ElimPulseStartTime2_CCPM") (Utilities|Time|GetGameTimeinSeconds))
    (Rendering|Material|SetScalarParameterValueonMaterials (CustomCharacter|GetSkeletalMeshForPartOnPlayerPawn) "ReactiveDuration_CCPM" (Variables|Default|GetReactivityAnimDuration))
    (Utilities|FlowControl|Sequence
      (:then_0
        (Utilities|FlowControl|DelayUntilNextTick)
        (Variables|Default|SetReactivityInProgress false))
      (:then_1
        (Variables|Default|SetAnimInProgress true)
        (bind _returnvalue (Utilities|Time|SetTimerbyEvent (delegate Custom|OnReactivityAnimFinished) (Variables|Default|GetReactivityAnimDuration)))
        (Variables|Default|SetReactivityAnimTimer _returnvalue)))))

(event Custom|OnReactivityAnimFinished
  (Variables|Default|SetAnimInProgress false))

(fn ConstructionScript
  (|CallParentFunction))
