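; Backpack part modifier for ZombieElastic that shows a hologram sticker through idle Niagara VFX and reacts to eliminations.
; Texture setup draws on the elastic character's body "Sticker" texture or backpack variants, feeding a dynamic material and Niagara.
; Textures over 256 pixels count as sprite sheets, animated on a loop in the frontend or on kill stat changes.
; Stat changes also drive Kill and Boost timelines into Niagara parameters.
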(event EventOnPostResIn
  (|CallParentFunction)
  (CallFunction|SetupTexture)
  ;[0.1338 0.9502 1] Watch Stat and play animation on reactivity or in loop if in frontend
  (if (Variables|Default|GetIsAnimated)
    (if (CustomCosmeticModifier|IsinFrontEnd)
      (bind (_direction _newtrack_0) (Timeline|StickerAnimationTL PlayFromStart)
        (:Update
          (Rendering|Material|SetScalarParameterValue (Variables|Default|GetInstanceMaterial) "AnimationPhase" _newtrack_0 :!class /Script/Engine.MaterialInstanceDynamic)
          (Niagara|SetNiagaraVariable(Float) (Utilities|Casting|CastToNiagaraComponent (CallFunction|GetIdleVFXComponent)) "AnimationPhase" _newtrack_0))
        (:Finished
          ;[0.1971 0.3908 1] Play In Loop In Front End
          (if (CustomCosmeticModifier|IsinFrontEnd)
            (Timeline|StickerAnimationTL PlayFromStart))))
      (else
        (Cosmetics|WatchCosmeticStat (CustomCharacter|GetAssociatedPlayerPawn) "(TagName=\"StatManager.AthenaKills\")" 0 false
          (:OnStatChanged
            (Timeline|StickerAnimationTL PlayFromStart)))))))

(event EventOnStatChanged
  (params (State%20Name :type Name :default "None") (Stat%20Value :type Integer))
  (bind _returnvalue (CallFunction|GetIdleVFXComponent))
  (bind _asniagara_particle_system_component (Utilities|Casting|CastToNiagaraComponent _returnvalue))
  ;[1 0.3856 0.2958] Elimination Reactivity
  (if (== Stat%20Value 0)
    _
    (elif (Variables|Default|GetbIsDarkStat)
      (Utilities|FlowControl|Sequence
        (:then_0
          (bind (_direction _newtrack_0) (Timeline|KillTL SetNewTime :NewTime 1.6)
            (:Update
              (Niagara|SetNiagaraVariable(Float) _asniagara_particle_system_component "KillingChange" _newtrack_0))
            (:Finished
              (Variables|Default|SetbIsDarkStat false))))
        (:then_1
          (bind (_direction_2 _newtrack_2) (Timeline|BoostTL SetNewTime)
            (:Update
              (Niagara|SetNiagaraVariable(Float) _asniagara_particle_system_component "Boost" _newtrack_2)))))
      (else
        (Variables|Default|SetbIsDarkStat true)
        (Utilities|FlowControl|Sequence
          (:then_0
            (Timeline|BoostTL PlayFromStart))
          (:then_1
            (Timeline|KillTL PlayFromStart)))))))

(event Custom|Setup%20Texture
  (bind _returnvalue (CallFunction|GetIdleVFXComponent))
  (bind _asniagara_particle_system_component (Utilities|Casting|CastToNiagaraComponent _returnvalue))
  ;[1 0.9482 0.1131] Texture Setup
  (Utilities|ShouldAllowElasticReactive
    (:Yes
      (if (CallFunction|IsCharacterElastic)
        (bind _result (ArrayMacros|SafeArrayGet (Rendering|Material|GetMaterials (Customization|GetSkeletalMeshforPartType (CustomCharacter|GetAssociatedPlayerPawn) "Body" :!class /Script/FortniteGame.FortPlayerPawn)))
          (:%20%20
            (bind _asmaterial_instance_dynamic (Utilities|Casting|CastToMaterialInstanceDynamic _result)
              (:then
                (bind _returnvalue_2 (Rendering|Material|GetTextureParameterValue _asmaterial_instance_dynamic "Sticker" :!class /Script/Engine.MaterialInstanceDynamic))
                (bind _returnvalue_3 (Rendering|Material|CreateDynamicMaterialInstance "/BRCosmetics/Characters/Player/Female/Medium/Bodies/F_MED_ZombieElasticNeon/FX/M_ZombieElastic_Holograms.M_ZombieElastic_Holograms" "Elastic Hologram MID" "None" :!class /Script/Engine.KismetMaterialLibrary))
                (Rendering|Material|SetTextureParameterValue _returnvalue_3 "2DHologramTexture" _returnvalue_2)
                (Niagara|SetNiagaraVariable(Material) _asniagara_particle_system_component "MID_ElasticHologram" _returnvalue_3)
                (Niagara|SetNiagaraVariable(Texture) _asniagara_particle_system_component "HologramTexture" _returnvalue_2)
                (CallFunction|SetupAnimatedMaterial))
              (:CastFailed
                (Development|PrintString "CCPM_Elastic_MID Cast FAILED")))))
        (else
          (Utilities|GateBackpackVariantFallback))))
    (:No
      _))
  ;;[1 1 1] First we need to make sure that we actually have a dynamic material...
  (bind _returnvalue_4 (Rendering|Material|CreateDynamicMaterialInstance "/BRCosmetics/Characters/Player/Female/Medium/Bodies/F_MED_ZombieElasticNeon/FX/M_ZombieElastic_Holograms.M_ZombieElastic_Holograms" "Elastic Hologram MID" "None" :!class /Script/Engine.KismetMaterialLibrary))
  (Variables|Default|SetInstanceMaterial _returnvalue_4)
  (Niagara|SetNiagaraVariable(Material) (Utilities|Casting|CastToNiagaraComponent (CallFunction|GetIdleVFXComponent)) "MID_ElasticHologram" _returnvalue_4)
  ;;[1 1 1] Now, we can have the backpack handle applying the variant.
  (Cosmetics|ApplyVariantsToComponent(Deprecated) (CosmeticLoadout|GetItemDefinitionfromLoadout (CosmeticLoadout|GetCosmeticLoadout) "(TagName=\"LoadoutSlot.Backpack\")") (CallFunction|GetIdleVFXComponent) (CallFunction|GrabLoadout) (Utilities|Struct|MakeApplyVariantsAdditionalParams))
  (Transformation|AttachComponentToComponent (CallFunction|GetIdleVFXComponent) (CustomCharacter|GetSkeletalMeshForPartOnPlayerPawn) (CallFunction|GetIdleFXSocketName) "SnapToTarget" "SnapToTarget" "SnapToTarget" true)
  (CallFunction|SetupAnimatedMaterial)
  (bind _returnvalue_5 (Rendering|Material|GetTextureParameterValue (Variables|Default|GetInstanceMaterial) "2DHologramTexture" :!class /Script/Engine.MaterialInstanceDynamic))
  (Niagara|SetNiagaraVariable(Texture) _asniagara_particle_system_component "HologramTexture" _returnvalue_5)
  (Development|PrintString (Utilities|GetDisplayName _returnvalue_5) true true "(R=1.000000,G=0.339068,B=0.000000,A=1.000000)"))

(fn GrabLoadout
  (return (Variables|/Script/FortniteGame.FortPlayerPawn|GetCosmeticLoadout (CustomCharacter|GetAssociatedPlayerPawn))))

(fn IsCharacterElastic
  (return (GameplayTags|HasTag (GameplayTags|GetOwnedGameplayTags (CosmeticLoadout|GetItemDefinitionfromLoadout (CosmeticLoadout|GetCosmeticLoadout) "(TagName=\"LoadoutSlot.Character\")") :!class /Script/GameplayTags.GameplayTagAssetInterface) "(TagName=\"Cosmetics.Set.Endless\")")))

(fn SetupAnimatedMaterial
  (bind _returnvalue (CallFunction|GetIdleVFXComponent))
  (bind _asniagara_particle_system_component (Utilities|Casting|CastToNiagaraComponent _returnvalue))
  (bind _instancematerial (Variables|Default|GetInstanceMaterial))
  (Utilities|IsValid _instancematerial
    (:Is%20Valid
      (bind _returnvalue_2 (Rendering|Material|GetTextureParameterValue _instancematerial "2DHologramTexture" :!class /Script/Engine.MaterialInstanceDynamic))
      ;[1 1 1] Check Texture Size to see if texure is a sprite sheet or not
      (bind _size (Class|VolumetricsBPLIbrary|GetTextureSize _returnvalue_2))
      (if (and (> (Utilities|Struct|BreakIntPoint _size) 256) (> (.Y (Utilities|Struct|BreakIntPoint _size)) 256))
        (Rendering|Material|SetScalarParameterValue _instancematerial "AnimatedHologram" 1.0 :!class /Script/Engine.MaterialInstanceDynamic)
        (Variables|Default|SetIsAnimated true)
        (Niagara|SetNiagaraVariable(Float) _asniagara_particle_system_component "AnimatedHologram" 1.0)
        (else
          (Rendering|Material|SetScalarParameterValue _instancematerial "AnimatedHologram" 0.0 :!class /Script/Engine.MaterialInstanceDynamic)
          (Variables|Default|SetIsAnimated false)
          (Niagara|SetNiagaraVariable(Float) _asniagara_particle_system_component "AnimatedHologram" 0.0))))))

(fn ConstructionScript
  (|CallParentFunction))
