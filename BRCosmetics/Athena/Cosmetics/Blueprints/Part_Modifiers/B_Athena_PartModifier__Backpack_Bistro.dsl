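; Cosmetic part modifier Blueprint for the Bistro backpack that responds when the pawn begins an emote.
; On emote begin it fetches the idle VFX component and activates it without a reset.
; The emote end event and construction script are empty, so nothing here deactivates the effect afterward.
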(event EventOnBeginEmote
  (params (MontageItemDef :type Fort%20Item%20Definition%20Object%20Reference) (PawnEmoting :type Fort%20Pawn%20Object%20Reference))
  (Components|Activation|Activate (CallFunction|GetIdleVFXComponent) false))

(event EventOnEndEmote
  (params (MontageItemDef :type Fort%20Item%20Definition%20Object%20Reference) (PawnEmoting :type Fort%20Pawn%20Object%20Reference)))

(fn ConstructionScript)
