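; Cosmetic part modifier for the Lazuli backpack that drives a reactive material and Niagara idle-VFX effect.
; After res-in it resets the backpack mesh's material parameters and, in the front end only, plays the StatLoop timeline into CCPM_EffectOn and Niagara Visibility.
; Any stat change with a positive value calls EffectOn, which clears FrontEnd, stamps CCPM_End_Time with game time and restarts a timeline driving Niagara Visibility.
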(event EventOnPostResIn
  (|CallParentFunction)
  (bind _output_get (Variables|Default|SetPlayerPawn (CustomCharacter|GetAssociatedPlayerPawn)))
  (bind _output_get_2 (Variables|Default|SetBodyMesh (Customization|GetSkeletalMeshforPartType _output_get "Backpack" :!class /Script/FortniteGame.FortPlayerPawn)))
  (Rendering|Material|SetScalarParameterValue (Class|BTechArtGlobalLibrary|GetDynamicMaterialInstance _output_get_2) "CCPM_EffectOn" 0.0 :!class /Script/Engine.MaterialInstanceDynamic)
  (Rendering|Material|SetScalarParameterValue (Class|BTechArtGlobalLibrary|GetDynamicMaterialInstance (Variables|Default|GetBodyMesh)) "FrontEnd" 1.0 :!class /Script/Engine.MaterialInstanceDynamic)
  (if (CustomCosmeticModifier|IsinFrontEnd)
    (bind (_direction _statechange) (Timeline|StatLoop Play)
      (:Update
        (Rendering|Material|SetScalarParameterValue (Class|BTechArtGlobalLibrary|GetDynamicMaterialInstance (Variables|Default|GetBodyMesh)) "CCPM_EffectOn" _statechange :!class /Script/Engine.MaterialInstanceDynamic)
        (bind _asniagara_particle_system_component (Utilities|Casting|CastToNiagaraComponent (CallFunction|GetIdleVFXComponent)))
        (Niagara|SetNiagaraVariable(Float) _asniagara_particle_system_component "Visibility" _statechange)))))

(event EventOnStatChanged
  (params (State%20Name :type Name :default "None") (Stat%20Value :type Integer))
  (|CallParentFunction State%20Name Stat%20Value)
  (if (> Stat%20Value 0)
    (CallFunction|EffectOn)))

(event EventOnDayPhaseChanged
  (params (CurrentDayPhase :type EFortDayPhase%20Enum :default "Morning") (PreviousDayPhase :type EFortDayPhase%20Enum :default "Morning") (bAtCreation :type Boolean)))

(event Custom|EffectOn
  (Rendering|Material|SetScalarParameterValue (Class|BTechArtGlobalLibrary|GetDynamicMaterialInstance (Variables|Default|GetBodyMesh)) "FrontEnd" 0.0 :!class /Script/Engine.MaterialInstanceDynamic)
  (Rendering|Material|SetScalarParameterValue (Class|BTechArtGlobalLibrary|GetDynamicMaterialInstance (Variables|Default|GetBodyMesh)) "CCPM_End_Time" (Utilities|Time|GetGameTimeinSeconds) :!class /Script/Engine.MaterialInstanceDynamic)
  (bind _asniagara_particle_system_component (Utilities|Casting|CastToNiagaraComponent (CallFunction|GetIdleVFXComponent)))
  (Utilities|IsValid _asniagara_particle_system_component
    (:Is%20Valid
      (bind (_direction _visibility) (Timeline|Timeline_0 PlayFromStart)
        (:Update
          (Niagara|SetNiagaraVariable(Float) _asniagara_particle_system_component "Visibility" _visibility))))))

(event EventOnNewPlayerWeaponType
  (params (NewType :type EFortItemType%20Enum :default "WorldItem") (OldType :type EFortItemType%20Enum :default "WorldItem"))
  (|CallParentFunction NewType OldType) ; This node is disabled and will not be called.\nDrag off pins to build functionality.
)

(event EventOnWeaponChanged
  (params (NewWeapon :type Fort%20Weapon%20Object%20Reference) (OldWeapon :type Fort%20Weapon%20Object%20Reference)))

(event EventOnWeaponFired)

(event EventOnNewWeapons
  (params (WeaponsAsItems :type Array%20of%20Fort%20Item%20Object%20References)))

(event EventOnCurieStateAttached
  (params (Tag :type Gameplay%20Tag%20Structure%20%28by%20ref%29)))

(event EventOnCurieStateDetached
  (params (Tag :type Gameplay%20Tag%20Structure%20%28by%20ref%29)))

(event EventOnBeginSkydiving
  (|CallParentFunction) ; This node is disabled and will not be called.\nDrag off pins to build functionality.
)

(event EventOnEndSkydiving
  (|CallParentFunction) ; This node is disabled and will not be called.\nDrag off pins to build functionality.
)

(event EventOnEnterVehicle
  (params (VehicleEntered :type Fort%20Vehicle%20Interface%20Interface%20%28by%20ref%29))
  (|CallParentFunction VehicleEntered) ; This node is disabled and will not be called.\nDrag off pins to build functionality.
)

(event EventOnExitVehicle
  (|CallParentFunction) ; This node is disabled and will not be called.\nDrag off pins to build functionality.
)

(event EventOnTargetingOrFirstPersonCameraChanged)

(event EventOnHitPawn
  (params (SelfActor :type Actor%20Object%20Reference) (OtherActor :type Actor%20Object%20Reference) (NormalImpulse :type Vector :default "0, 0, 0") (Hit :type Hit%20Result%20Structure%20%28by%20ref%29)))

(fn ConstructionScript
  (|CallParentFunction))
