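; Cosmetic part modifier for the RhombCamo backpack that shows the player's elimination count through dynamic material parameters on the backpack mesh.
; Watched stat changes set the counter offset, and a positive count plays a UV-offset timeline that clears a first-trigger flag when finished.
; In the front end, a looping timeline demos the counter and plays or reverses that animation.
; When customization finishes, it re-reads the stat and snaps a positive count to fixed offsets.
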(event EventOnPostResIn
  (|CallParentFunction)
  (bind _output_get (Variables|Default|SetPlayerPawn (CustomCharacter|GetAssociatedPlayerPawn)))
  (bind _output_get_2 (Variables|Default|SetBodyMesh (Customization|GetSkeletalMeshforPartType _output_get "Backpack" :!class /Script/FortniteGame.FortPlayerPawn)))
  (Rendering|Material|SetScalarParameterValue (Class|BTechArtGlobalLibrary|GetDynamicMaterialInstance _output_get_2) "Elimination_CounterOffset" 0.0 :!class /Script/Engine.MaterialInstanceDynamic)
  (Variables|Default|SetFirstTriggered true)
  (if (CustomCosmeticModifier|IsinFrontEnd)
    (bind (_direction _counterint) (Timeline|LoopingStatue Play)
      (:Update
        (CallFunction|StageOne (Math|Float|Truncate _counterint)))
      (:ReverseUV
        (CallFunction|UVReverseOperation))
      (:ChangeUV
        (CallFunction|UVOperation)))))

(event EventOnStatChanged
  (params (State%20Name :type Name :default "None") (Stat%20Value :type Integer))
  (bind _returnvalue (Math|Conversions|ToFloat(Integer) Stat%20Value))
  (|CallParentFunction State%20Name Stat%20Value)
  (Variables|Default|SetPlayerKillCounter Stat%20Value)
  (Rendering|Material|SetScalarParameterValue (Class|BTechArtGlobalLibrary|GetDynamicMaterialInstance (Variables|Default|GetBodyMesh)) "Elimination_CounterOffset" _returnvalue :!class /Script/Engine.MaterialInstanceDynamic)
  (if (> _returnvalue 0)
    (if (Variables|Default|GetFirstTriggered)
      (CallFunction|UVOperation))))

(event Custom|StageOne
  (params (CounterLooping :type Integer))
  (Rendering|Material|SetScalarParameterValue (Class|BTechArtGlobalLibrary|GetDynamicMaterialInstance (Variables|Default|GetBodyMesh)) "Elimination_CounterOffset" (Math|Conversions|ToFloat(Integer) CounterLooping) :!class /Script/Engine.MaterialInstanceDynamic))

(event EventOnCharacterCustomizationFinished
  (params (CharacterPartInfo :type Array%20of%20Applied%20Character%20Part%20Info%20Structures))
  (bind _bodymesh (Variables|Default|GetBodyMesh))
  (bind _material_instance_dynamic (Class|BTechArtGlobalLibrary|GetDynamicMaterialInstance _bodymesh))
  ;[1 1 1] WHen player change cosmetic in game, it will call a growing up number function.
  (Development|PrintString "Change")
  (bind (_statname _statvalue) (Cosmetics|WatchCosmeticStat (CustomCharacter|GetAssociatedPlayerPawn) (Variables|PartModifierControls|WatchCosmeticStat|GetWatch%20Cosmetic%20Stat%20Tag) -1 true -1.0 (Class|FortCosmeticFunctionLibrary|GetCosmeticStatDelay (Variables|PartModifierControls|WatchCosmeticStat|GetWatch%20Cosmetic%20Stat%20Tag)) (Variables|Default|GetConsume%20Waiting%20Notifications))) ; Everytime you change the skin, the local variable you stored will be destoried due to lifetime out
  (if (> _statvalue 0)
    (Rendering|Material|SetScalarParameterValue _material_instance_dynamic "Elimination_CounterOffset" (Math|Conversions|ToFloat(Integer) _statvalue) :!class /Script/Engine.MaterialInstanceDynamic)
    (Rendering|Material|SetScalarParameterValue _material_instance_dynamic "Shape_UVOffsetX" -0.76 :!class /Script/Engine.MaterialInstanceDynamic)
    (Rendering|Material|SetScalarParameterValue _material_instance_dynamic "Shape_UVOffsetY" -0.34 :!class /Script/Engine.MaterialInstanceDynamic)))

(event Custom|UVOperation
  (bind _bodymesh (Variables|Default|GetBodyMesh))
  (bind _material_instance_dynamic (Class|BTechArtGlobalLibrary|GetDynamicMaterialInstance _bodymesh))
  (bind (_direction _uv_x _uv_y) (Timeline|UVAnimation Play)
    (:Update
      (Rendering|Material|SetScalarParameterValue _material_instance_dynamic "Shape_UVOffsetX" _uv_x :!class /Script/Engine.MaterialInstanceDynamic)
      (Rendering|Material|SetScalarParameterValue _material_instance_dynamic "Shape_UVOffsetY" _uv_y :!class /Script/Engine.MaterialInstanceDynamic))
    (:Finished
      (Variables|Default|SetFirstTriggered false))))

(event Custom|UVReverseOperation
  (bind _bodymesh (Variables|Default|GetBodyMesh))
  (bind _material_instance_dynamic (Class|BTechArtGlobalLibrary|GetDynamicMaterialInstance _bodymesh))
  (Timeline|UVAnimation Reverse))

(fn ConstructionScript
  (|CallParentFunction))
