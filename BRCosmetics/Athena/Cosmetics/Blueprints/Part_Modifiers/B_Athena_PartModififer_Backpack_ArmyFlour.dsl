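; Part modifier Blueprint for the ArmyFlour back bling that feeds a kill count into its idle particle effect.
; On stat changes, it writes the incoming value to the idle VFX component's "KillCount" parameter, without filtering by stat name.
; After post-resource initialization, if frontend reactivity should display, it waits briefly, increments a frontend kill count and pushes that to the same parameter as a preview, followed by a trailing delay.
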(event EventOnPostResIn
  (|CallParentFunction)
  (if (CallFunction|ShouldDisplayFrontendReactivity)
    (Utilities|FlowControl|Delay 1.0)
    (bind _result (Math|Integer|IncrementInt (Variables|Default|GetFEKillCount))
      (:%20%20
        (Effects|Components|ParticleSystem|SetIntParameter (CallFunction|GetIdleVFXComponent) "KillCount" _result)
        (Utilities|FlowControl|Delay 4.0)))))

(event EventOnStatChanged
  (params (State%20Name :type Name :default "None") (Stat%20Value :type Integer))
  (|CallParentFunction State%20Name Stat%20Value)
  (Effects|Components|ParticleSystem|SetIntParameter (CallFunction|GetIdleVFXComponent) "KillCount" Stat%20Value))

(fn ConstructionScript
  (|CallParentFunction))
