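; Part modifier Blueprint for the EmberYearn back bling cosmetic, adjusting its idle VFX and supplying a sound.
; After the parent res-in event, it sets the idle particle component's bIsMale parameter from the associated player pawn's gender.
; SetAndPlaySound spawns the Apprentice back bling elimination-open cue attached to the part's skeletal mesh; nothing in this Blueprint calls it.
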(event EventOnPostResIn
  (|CallParentFunction)
  (Effects|Components|ParticleSystem|SetBoolParameter (CallFunction|GetIdleVFXComponent) "bIsMale" (FortPlayerPawn|GetGender (CustomCharacter|GetAssociatedPlayerPawn))))

(fn SetAndPlaySound
  (Audio|SpawnSoundAttached "/BRCosmetics/Sounds/Backblings/Apprentice/BackBling_Apprentice_Elimination_Open_Cue.BackBling_Apprentice_Elimination_Open_Cue" (CustomCharacter|GetSkeletalMeshForPartOnPlayerPawn)))

(fn ConstructionScript
  (|CallParentFunction))
