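; Backpack part modifier for the CombCrater cosmetic that reacts to the player's eliminations with material and VFX changes.
; After post-res-in, it watches the AthenaKills stat; a value above zero stamps game time into the mesh material's BP_GetTime and activates the idle Niagara VFX, otherwise deactivating it.
; When frontend reactivity should display, a one-shot delayed sequence previews the same effect, then turns the VFX off after a further delay.
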(event EventOnPostResIn
  (bind _returnvalue (CallFunction|GetIdleVFXComponent))
  (|CallParentFunction)
  (Utilities|FlowControl|Sequence
    (:then_0
      ;[1 0.7262 0.2176 1 60] On Kill Reactivity ☠
      (bind (_statname _statvalue) (Cosmetics|WatchCosmeticStat (CustomCharacter|GetAssociatedPlayerPawn) "(TagName=\"StatManager.AthenaKills\")" 0 false)
        (:OnStatChanged
          (if (> _statvalue 0)
            ;;[1 1 1] Material Reactivity
            (Rendering|Material|SetScalarParameterValueonMaterials (CustomCharacter|GetSkeletalMeshForPartOnPlayerPawn) "BP_GetTime" (Utilities|Time|GetGameTimeinSeconds))
            (bind _asniagara_particle_system_component (Utilities|Casting|CastToNiagaraComponent _returnvalue))
            (Components|Activation|Activate _asniagara_particle_system_component true)
            (else
              (bind _asniagara_particle_system_component_2 (Utilities|Casting|CastToNiagaraComponent _returnvalue))
              (Components|Activation|Deactivate _asniagara_particle_system_component_2))))))
    (:then_1
      ;[0.141 0.4403 1 1 60] Frontend 🛒
      (bind _output_get (Variables|Default|SetFE_Reactive (Reactive|Frontend|ShouldDisplayFrontendReactivity 0 "Backpack")))
      (if _output_get
        (Utilities|FlowControl|DoOnce false
          (:Completed
            (Utilities|FlowControl|Delay 4.0)
            (Rendering|Material|SetScalarParameterValueonMaterials (CustomCharacter|GetSkeletalMeshForPartOnPlayerPawn) "BP_GetTime" (Utilities|Time|GetGameTimeinSeconds))
            (bind _asniagara_particle_system_component_3 (Utilities|Casting|CastToNiagaraComponent (CallFunction|GetIdleVFXComponent)))
            (Components|Activation|Activate _asniagara_particle_system_component_3 true)
            (Utilities|FlowControl|Delay 12.2)
            (Components|Activation|Deactivate _asniagara_particle_system_component_3)))))))

(fn ConstructionScript
  (|CallParentFunction))
