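; Part modifier for the ShinyStar backpack cosmetic that drives its reactive material effect.
; After res-in, it checks whether frontend reactivity should display and, if so, plays the frontend reactivity timeline.
; Otherwise it watches the Athena kills stat on the associated player pawn and restarts an in-game timeline on each change.
; Both timelines write the ReactivityBoost scalar parameter to the part's skeletal mesh materials.
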(event EventOnPostResIn
  (|CallParentFunction)
  (if (Reactive|Frontend|ShouldDisplayFrontendReactivity 0 "Backpack")
    ;[0 0 0] FE cycling
    (bind (_direction _reactivity) (Timeline|FE%20reactivity PlayFromStart)
      (:Update
        (Rendering|Material|SetScalarParameterValueonMaterials (CustomCharacter|GetSkeletalMeshForPartOnPlayerPawn) "ReactivityBoost" _reactivity)))
    (else
      (Cosmetics|WatchCosmeticStat (CustomCharacter|GetAssociatedPlayerPawn) "(TagName=\"StatManager.AthenaKills\")" 0 false
        (:OnStatChanged
          (bind (_direction_2 _starsspeedboost) (Timeline|Ingame%20reactivity PlayFromStart)
            (:Update
              (Rendering|Material|SetScalarParameterValueonMaterials (CustomCharacter|GetSkeletalMeshForPartOnPlayerPawn) "ReactivityBoost" _starsspeedboost))))))))

(fn ConstructionScript
  (|CallParentFunction))
