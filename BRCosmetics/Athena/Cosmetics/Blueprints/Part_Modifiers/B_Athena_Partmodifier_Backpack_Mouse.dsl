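; Backpack part modifier for the Mouse back bling that flashes a reactive Niagara effect on player kills.
; After res-in, it loops a six-second preview timer in the frontend, and otherwise watches the AthenaKills stat for values above zero.
; Both paths spawn the ActiveFlash system on the dyn_pack socket, apply backpack variants to it, and run a timeline.
; The timeline drives the StateLerp parameter on the idle VFX component.
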(event Custom|FrontEnd%20Cycle
  (bind _returnvalue (PartModifier|SpawnSystem|SpawnSystemAttachedPartModifier "/BRCosmetics/Accessories/FORT_Backpacks/Backpack_F_MED_Mouse/FX/Systems/NS_Backpack_Mouse_ActiveFlash.NS_Backpack_Mouse_ActiveFlash" (CustomCharacter|GetSkeletalMeshForPartOnPlayerPawn) "dyn_pack")
    (:Success
      (Class|FortCosmeticFunctionLibrary|ApplyVariantstoComponentbySlotTag (CustomCharacter|GetAssociatedPlayerPawn) "(TagName=\"LoadoutSlot.Backpack\")" _returnvalue
        (:True
          (bind (_direction _lerpin) (Timeline|Timeline Play)
            (:Update
              (Effects|Components|ParticleSystem|SetFloatParameter (Utilities|Casting|CastToNiagaraComponent (CallFunction|GetIdleVFXComponent)) "StateLerp" _lerpin))))
        (:False
          (Timeline|Timeline Play))))
    (:Failure
      (Timeline|Timeline PlayFromStart))))

(event EventOnPostResIn
  (|CallParentFunction)
  (if (Reactive|Frontend|IsCosmeticObjectDisplayedinFrontendContext 0)
    (Utilities|Time|SetTimerbyEvent (EventDispatchers|CreateEvent :!event Custom|FrontEnd%20Cycle) 6.0 true)
    (else
      (bind (_statname _statvalue) (Cosmetics|WatchCosmeticStat (CustomCharacter|GetAssociatedPlayerPawn) "(TagName=\"StatManager.AthenaKills\")" 0 false)
        (:OnStatChanged
          (if (> _statvalue 0)
            (bind _returnvalue (PartModifier|SpawnSystem|SpawnSystemAttachedPartModifier "/BRCosmetics/Accessories/FORT_Backpacks/Backpack_F_MED_Mouse/FX/Systems/NS_Backpack_Mouse_ActiveFlash.NS_Backpack_Mouse_ActiveFlash" (CustomCharacter|GetSkeletalMeshForPartOnPlayerPawn) "dyn_pack")
              (:Success
                (Class|FortCosmeticFunctionLibrary|ApplyVariantstoComponentbySlotTag (CustomCharacter|GetAssociatedPlayerPawn) "(TagName=\"LoadoutSlot.Backpack\")" _returnvalue
                  (:True
                    (Timeline|Timeline Play))
                  (:False
                    (Timeline|Timeline Play))))
              (:Failure
                (Timeline|Timeline PlayFromStart)))))))))

(fn ConstructionScript
  (|CallParentFunction))
