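; Cosmetic Blueprint for the ScoopPluck pickaxe, adding reactive visuals and sound on eliminations and impacts.
; Outside preview it watches the owner's kill stat, while impacts react on listed surface types or in preview, each trigger with its own sound.
; Reactivity relays through server and client events to increment Eliminations, feed it to the idle VFX and timestamp both weapon meshes' materials.
; The passed kill count is unused, so impacts also raise the counter.
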(event EventOnInitWeaponCosmetics
  (|CallParentFunction)
  (if (not (Reactive|Frontend|IsCosmeticObjectDisplayedasPreview 0))
    (bind _asfort_player_pawn (Utilities|Casting|CastToFortPlayerPawn (Variables|Actor|GetInstigator)))
    (bind (_statname _statvalue) (Cosmetics|WatchCosmeticStat _asfort_player_pawn "(TagName=\"StatManager.AthenaKills\")" 0 false)
      (:OnStatChanged
        (Utilities|FlowControl|Sequence
          (:then_1
            (if (>= _statvalue 1)
              (CallFunction|Reactivity _statvalue)
              (Audio|SpawnSoundAttached "/BRCosmetics/Sounds/PickAxes/ScoopPluck/MSS_Pickaxe_ScoopPluck_Elim_Reactivity.MSS_Pickaxe_ScoopPluck_Elim_Reactivity" (Variables|Visual|GetWeaponMesh)))))))))

(event Custom|Client%20Reactivity
  (bind _returnvalue (Utilities|Time|GetGameTimeinSeconds))
  (bind _result (Math|Integer|IncrementInt (Variables|Default|GetEliminations))
    (:%20%20
      (Effects|Components|ParticleSystem|SetIntParameter (CallFunction|GetIdleVFXComponent) "KillReactivityCount" _result)
      (Rendering|Material|SetScalarParameterValueonMaterials (Variables|Visual|GetWeaponMesh) "TimeStart" _returnvalue)
      (Rendering|Material|SetScalarParameterValueonMaterials (Variables|Visual|GetWeaponMeshOffhand) "TimeStart" _returnvalue))))

(event Custom|ServerReactivity
  (CallFunction|ClientReactivity))

(event Custom|Reactivity
  (params (KillCount :type Integer))
  (CallFunction|ServerReactivity))

(event FortWeapon|EventOnPlayImpactFX
  (params (HitResult :type Hit%20Result%20Structure%20%28by%20ref%29) (ImpactPhysicalSurface :type EPhysicalSurface%20Enum :default "SurfaceType_Default") (SpawnedPSC :type FXSystem%20Component%20Object%20Reference))
  (|CallParentFunction HitResult ImpactPhysicalSurface SpawnedPSC)
  (if (Utilities|Array|ContainsItem (Variables|Default|GetImpactPhysicalSurface) ImpactPhysicalSurface)
    _
    (elif (Reactive|Frontend|IsCosmeticObjectDisplayedasPreview 0)
      _))
  (CallFunction|Reactivity 0)
  (Audio|SpawnSoundAttached "/BRCosmetics/Sounds/PickAxes/ScoopPluck/MSS_Pickaxe_ScoopPluck_Reactivity.MSS_Pickaxe_ScoopPluck_Reactivity" (Variables|Visual|GetWeaponMesh)))

(fn ConstructionScript
  (|CallParentFunction))
