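; Cosmetic part modifier for the TurnVortex Brink character part that animates a reactive glow in the front end and on kills.
; After EventOnPostResIn, the front end starts a glow timeline following a retriggerable delay, which the ReplayCountDown custom event can also replay.
; A watch on the pawn's kills stat plays a second timeline whenever the changed value is above zero.
; Both timelines drive the GlowReactive scalar on the part's skeletal mesh materials.
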(event Custom|ReplayCountDown
  ;[0.141 0.4403 1 1 200] FRONTEND 🛒
  (Utilities|FlowControl|RetriggerableDelay 3.0)
  (bind (_direction _animationa) (Timeline|Timeline_0 PlayFromStart)
    (:Update
      (Rendering|Material|SetScalarParameterValueonMaterials (CustomCharacter|GetSkeletalMeshForPartOnPlayerPawn) "GlowReactive" _animationa))))

(event EventBeginPlay)

(event EventOnPostResIn
  ;[1 0.2176 0.2574 1 200] EVENT 🖥️
  (|CallParentFunction)
  (Utilities|FlowControl|Sequence
    (:then_0
      ;[0.141 0.4403 1 1 200] FRONTEND 🛒
      (if (CustomCosmeticModifier|IsinFrontEnd)
        (Utilities|FlowControl|RetriggerableDelay 3.0)
        (Timeline|Timeline_0 PlayFromStart)))
    (:then_1
      ;[1 0.7262 0.2176 1 200] ON KILL REACTIVITY ☠
      (bind (_statname _statvalue) (Cosmetics|WatchCosmeticStat (CustomCharacter|GetAssociatedPlayerPawn) "(TagName=\"StatManager.AthenaKills\")" 0 false)
        (:OnStatChanged
          (if (> _statvalue 0)
            (bind (_direction _animationa) (Timeline|Timeline PlayFromStart)
              (:Update
                (Rendering|Material|SetScalarParameterValueonMaterials (CustomCharacter|GetSkeletalMeshForPartOnPlayerPawn) "GlowReactive" _animationa)))))))))

(fn ConstructionScript
  (|CallParentFunction))
