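; Back bling cosmetic modifier for CHimeCurlTell that triggers the backpack's idle VFX component.
; A positive stat change activates the VFX through a DoOnce with no visible reset; the retriggerable delay after it drives nothing.
; In the front end, post res-in starts a cosmetic-only looping timer that re-activates the VFX, with a negative start delay bringing the first burst forward.
; The timer is cleared on destruction.
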(event EventOnStatChanged
  (params (State%20Name :type Name :default "None") (Stat%20Value :type Integer))
  (|CallParentFunction State%20Name Stat%20Value)
  (if (> Stat%20Value 0)
    (Utilities|FlowControl|DoOnce false
      (:Completed
        (Components|Activation|Activate (CallFunction|GetIdleVFXComponent) true)
        (Utilities|FlowControl|RetriggerableDelay 1.5)))))

(event EventOnPostResIn
  (|CallParentFunction)
  (if (CustomCosmeticModifier|IsinFrontEnd)
    (Networking|CanExecuteCosmeticEvents
      (:True
        (bind _returnvalue (Utilities|Time|SetTimerbyEvent (delegate Custom|CE%20FN%20Loop) 6.0 true false -5.0))
        (Variables|Default|SetTimer _returnvalue)))))

(event Custom|CE%20FN%20Loop
  (Components|Activation|Activate (CallFunction|GetIdleVFXComponent) true))

(event EventDestroyed
  (|CallParentFunction)
  (if (CustomCosmeticModifier|IsinFrontEnd)
    (Utilities|Time|ClearandInvalidateTimerbyHandle (Variables|Default|GetTimer))))

(fn ConstructionScript
  (|CallParentFunction))
