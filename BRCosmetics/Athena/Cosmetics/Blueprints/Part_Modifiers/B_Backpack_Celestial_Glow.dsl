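; Cosmetic part modifier for the Celestial backpack that flares its idle particle glow when a tracked stat, apparently kills, changes above zero.
; If idle, it reactivates the VFX with OnKillBool set, and timelines blend OnKillMaterialBlend to the kill state, then back after three seconds.
; The bIsIdle flag blocks retriggers until the return finishes or the stat drops to zero.
; In the frontend, a looping timer replays the kill reaction as a showcase.
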(event EventOnStatChanged
  (params (State%20Name :type Name :default "None") (Stat%20Value :type Integer))
  (bind _returnvalue (CallFunction|GetIdleVFXComponent))
  (CallFunction|DEBUGPrintString (Utilities|String|Append (Utilities|String|ToString(Name) State%20Name) "..." (Utilities|String|ToString(Integer) Stat%20Value)))
  (if (> Stat%20Value 0)
    (if (Variables|Default|GetbIsIdle)
      (Variables|Default|SetbIsIdle false)
      (Effects|Components|ParticleSystem|SetBoolParameter _returnvalue "OnKillBool" true)
      (Components|Activation|Activate _returnvalue true)
      (bind (_direction _blend) (Timeline|ToKillState PlayFromStart)
        (:Update
          (Effects|Components|ParticleSystem|SetFloatParameter _returnvalue "OnKillMaterialBlend" _blend))
        (:Finished
          (Utilities|FlowControl|Delay 3.0)
          (bind (_direction_2 _blend_2) (Timeline|ToIdleState PlayFromStart)
            (:Update
              (Effects|Components|ParticleSystem|SetFloatParameter _returnvalue "OnKillMaterialBlend" _blend_2))
            (:Finished
              (Effects|Components|ParticleSystem|SetBoolParameter _returnvalue "OnKillBool" false)
              (Variables|Default|SetbIsIdle true))))))
    (else
      (Variables|Default|SetbIsIdle true))))

(event Custom|Frontend_Event
  (bind _returnvalue (CallFunction|GetIdleVFXComponent))
  (Effects|Components|ParticleSystem|SetBoolParameter _returnvalue "OnKillBool" true)
  (Components|Activation|Activate _returnvalue true)
  (Timeline|ToKillState PlayFromStart))

(event EventOnPostResIn
  ;[1 1 1] Frontend_CyclingReactivity
  (|CallParentFunction)
  (if (CustomCosmeticModifier|IsinFrontEnd)
    (Utilities|FlowControl|Delay 0.2)
    (Utilities|Time|SetTimerbyEvent (EventDispatchers|CreateEvent :!event Custom|Frontend_Event) 1e+01 true)
    (else
      (Variables|Default|SetbIsIdle true))))

(fn ConstructionScript
  (|CallParentFunction))
