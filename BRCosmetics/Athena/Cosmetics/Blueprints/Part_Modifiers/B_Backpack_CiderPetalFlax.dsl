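; Cosmetic part modifier for the CiderPetalFlax backpack that plays material and Niagara reactions to the owner's eliminations and damage taken.
; Watched kill and damage stats each pass a DoOnce, stamp a game-time material parameter, spawn an attached burst, then delay.
; In the front end, a delay chain timed to the AnimBP activates the idle VFX and stamps the elimination timestamp.
; Destruction clears a timer handle in the front end.
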(event EventOnPostResIn
  (bind _returnvalue (CustomCharacter|GetAssociatedPlayerPawn))
  (bind _returnvalue_2 (CallFunction|GetIdleVFXComponent))
  (bind _returnvalue_3 (Utilities|Time|GetGameTimeinSeconds))
  (bind _returnvalue_4 (CustomCharacter|GetSkeletalMeshforPartonOwner))
  (|CallParentFunction)
  (Utilities|FlowControl|Sequence
    (:then_0
      (bind (_statname _statvalue) (Cosmetics|WatchCosmeticStat _returnvalue "(TagName=\"StatManager.AthenaKills\")" -1)
        (:OnStatChanged
          (if (> _statvalue 0)
            (Utilities|FlowControl|DoOnce false
              (:Completed
                (Rendering|Material|SetScalarParameterValueonMaterials _returnvalue_4 "Timestamp_Elim" _returnvalue_3)
                (Niagara|SpawnSystemAttached "/BRCosmetics/Accessories/FORT_Backpacks/Backpack_M_MED_CiderPetalFlax/FX/NS_Backpack_CiderPetalFlax_Elim_Burst.NS_Backpack_CiderPetalFlax_Elim_Burst" _returnvalue_2)
                (Utilities|FlowControl|Delay 2.85)))))))
    (:then_1
      (bind (_statname_2 _statvalue_2) (Cosmetics|WatchCosmeticStat _returnvalue "(TagName=\"StatManager.DamageReceived\")" -1)
        (:OnStatChanged
          (if (> _statvalue_2 0)
            (Utilities|FlowControl|DoOnce false
              (:Completed
                (Rendering|Material|SetScalarParameterValueonMaterials _returnvalue_4 "Timestamp_Dmg" _returnvalue_3)
                (Niagara|SpawnSystemAttached "/BRCosmetics/Accessories/FORT_Backpacks/Backpack_M_MED_CiderPetalFlax/FX/NS_Backpack_CiderPetalFlax_Damage.NS_Backpack_CiderPetalFlax_Damage" _returnvalue_2)
                (Utilities|FlowControl|Delay 1.5)))))))
    (:then_2
      (if (CustomCosmeticModifier|IsinFrontEnd)
        (Networking|CanExecuteCosmeticEvents
          (:True
            ; Match AnimBP timing
            (Utilities|FlowControl|Delay 1.0)
            (CallFunction|CEFNLoop)
            (Utilities|FlowControl|Delay 2.8)
            (Utilities|FlowControl|Delay 3.18)
            (CallFunction|CEFNLoop)
            (Utilities|FlowControl|Delay 3.0)
            (Utilities|FlowControl|Delay 2.95)))))))

(event Custom|CE%20FN%20Loop
  (Components|Activation|Activate (CallFunction|GetIdleVFXComponent) true)
  (Rendering|Material|SetScalarParameterValueonMaterials (CustomCharacter|GetSkeletalMeshforPartonOwner) "Timestamp_Elim" (Utilities|Time|GetGameTimeinSeconds)))

(event EventDestroyed
  (|CallParentFunction)
  (if (CustomCosmeticModifier|IsinFrontEnd)
    (Utilities|Time|ClearandInvalidateTimerbyHandle (Variables|Default|GetTimer))))

(fn ConstructionScript
  (|CallParentFunction))
