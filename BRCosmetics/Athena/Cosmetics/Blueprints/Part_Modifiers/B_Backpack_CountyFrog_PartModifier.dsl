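; Part modifier for the CountyFrog back bling that animates material-driven expressions, idle and reactive to healing.
; The idle timeline ping-pongs with random transition and hold times, while material parameters flag each stepped pose change.
; A GainedHealthTimes stat change blends out the idle, plays one of two random reactive expressions, then blends back.
; In frontend or preview, a recursive delay loop replays the healing reaction instead of watching the stat.
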(event EventOnPostResIn
  (|CallParentFunction)
  (if (or (Reactive|Frontend|IsCosmeticObjectDisplayedinFrontendContext 0) (Reactive|Frontend|IsCosmeticObjectDisplayedasPreview 0))
    (CallFunction|FrontendLoop)
    (else
      (Utilities|FlowControl|Sequence
        (:then_0
          (Cosmetics|WatchCosmeticStat (CustomCharacter|GetAssociatedPlayerPawn) "(TagName=\"StatManager.GainedHealthTimes\")" 0 false
            (:OnStatChanged
              (CallFunction|PlayHealingReactivity))))
        (:then_1
          (CallFunction|PlayIdleExpression))))))

(event Custom|PlayIdleExpression
  (bind _currentidleexpression (Variables|Default|GetCurrentIdleExpression))
  (bind _returnvalue (* _currentidleexpression 6.0))
  (bind _returnvalue_2 (Math|Float|Floor _returnvalue))
  (Utilities|FlowControl|Sequence
    (:then_0
      (Components|Timeline|SetPlayRate (Variables|B_Backpack_CountyFrog_PartModifier|GetIdlePoseCycle) (/ 1.0 (Math|Random|RandomFloatinRange (Variables|Default|GetIdleTransitionMinTime) (Variables|Default|GetIdleTransitionMaxTime)))))
    (:then_1
      (bind (_direction _expression) (Timeline|IdlePoseCycle Play)
        (:Update
          (Variables|Default|SetCurrentIdleExpression _expression)
          (Rendering|Material|SetScalarParameterValueonMaterials (CustomCharacter|GetSkeletalMeshForPartOnPlayerPawn) "ExpressionAmount" _currentidleexpression)
          (Rendering|Material|SetScalarParameterValueonMaterials (CustomCharacter|GetSkeletalMeshForPartOnPlayerPawn) "HasIdlePoseChanged" (Math|Conversions|ToFloat(Boolean) (!= _returnvalue_2 (Variables|Default|GetPreviousIdlePose))))
          (Variables|Default|SetPreviousIdlePose _returnvalue_2))
        (:Finished
          (Utilities|FlowControl|Delay (Math|Random|RandomFloatinRange (Variables|Default|GetIdleMinHoldTime) (Variables|Default|GetIdleMaxHoldTime)))
          (switch Utilities|FlowControl|Switch|SwitchonETimelineDirection _direction
            (:Forward
              (CallFunction|PlayIdleExpressionBackward))
            (:Backward
              (CallFunction|PlayIdleExpression))))))))

(event Custom|PlayIdleExpressionBackward
  (bind _currentidleexpression (Variables|Default|GetCurrentIdleExpression))
  (bind _returnvalue (* _currentidleexpression 6.0))
  (bind _returnvalue_2 (Math|Float|Floor _returnvalue))
  (Utilities|FlowControl|Sequence
    (:then_0
      (Components|Timeline|SetPlayRate (Variables|B_Backpack_CountyFrog_PartModifier|GetIdlePoseCycle) (/ 1.0 (Math|Random|RandomFloatinRange (Variables|Default|GetIdleTransitionMinTime) (Variables|Default|GetIdleTransitionMaxTime)))))
    (:then_1
      (Timeline|IdlePoseCycle Reverse))))

(event Custom|PlayHealingReactivity
  (bind _currentidleexpression (Variables|Default|GetCurrentIdleExpression))
  (if (not (Variables|Default|GetIsReactivityActive))
    (bind (_direction _overrideidle) (Timeline|InitHealingReactivity Play)
      (:Update
        (Rendering|Material|SetScalarParameterValueonMaterials (CustomCharacter|GetSkeletalMeshForPartOnPlayerPawn) "PlayReactivity" _overrideidle)
        (Rendering|Material|SetScalarParameterValueonMaterials (CustomCharacter|GetSkeletalMeshForPartOnPlayerPawn) "HasTransitionPoseChanged" (Math|Conversions|ToFloat(Boolean) (!= (Math|Float|Floor (* (Math|Float|Min(Float) _currentidleexpression (- 1.0 _overrideidle)) 6.0)) (Variables|Default|GetPreviousBlendPose))))
        (Variables|Default|SetPreviousBlendPose (Math|Float|Floor (* (Math|Float|Min(Float) _currentidleexpression (- 1.0 _overrideidle)) 6.0))))
      (:Finished
        (switch Utilities|FlowControl|Switch|SwitchonETimelineDirection _direction
          (:Forward
            (Variables|Default|SetIsReactivityActive true)
            (Rendering|Material|SetScalarParameterValueonMaterials (CustomCharacter|GetSkeletalMeshForPartOnPlayerPawn) "ReactiveExpressionID" (Math|Conversions|ToFloat(Boolean) (Math|Random|RandomBool)))
            (bind (_direction_2 _reactivityamt) (Timeline|ActiveHealingReactivity PlayFromStart)
              (:Update
                (Rendering|Material|SetScalarParameterValueonMaterials (CustomCharacter|GetSkeletalMeshForPartOnPlayerPawn) "ReactiveExpressionAmount" _reactivityamt)
                (Rendering|Material|SetScalarParameterValueonMaterials (CustomCharacter|GetSkeletalMeshForPartOnPlayerPawn) "HasReactivePoseChanged" (Math|Conversions|ToFloat(Boolean) (!= (Math|Float|Floor (* _reactivityamt 6.0)) (Variables|Default|GetPreviousReactivePose))))
                (Variables|Default|SetPreviousReactivePose (Math|Float|Floor (* _reactivityamt 6.0))))
              (:Finished
                (Variables|Default|SetIsReactivityActive false)
                (CallFunction|EndHealingReactivity)))))))))

(event Custom|EndHealingReactivity
  (bind _currentidleexpression (Variables|Default|GetCurrentIdleExpression))
  (Timeline|InitHealingReactivity Reverse))

(event Custom|FrontendLoop
  (CallFunction|PlayIdleExpression)
  (Utilities|FlowControl|Delay 5.0)
  (CallFunction|PlayHealingReactivity)
  (Utilities|FlowControl|Delay 3.16)
  (CallFunction|PlayHealingReactivity)
  (Utilities|FlowControl|Delay 3.3)
  (CallFunction|FrontendLoop))

(fn ConstructionScript
  (|CallParentFunction))
