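; Cosmetic part modifier for the LastVoiceDive back bling that plays a reaction sound when the player's kill stat changes.
; On post-resource init it watches the Athena kills stat and enables reactivity based on a valid Stage1 match in the Parts variant channel.
; When enabled, the kill handler spawns the sound attached to the part's skeletal mesh.
; A helper sets an integer audio parameter but is not called here.
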(event EventOnPostResIn
  (|CallParentFunction)
  (Utilities|FlowControl|Sequence
    (:then_0
      (Cosmetics|WatchCosmeticStat (CustomCharacter|GetAssociatedPlayerPawn) "(TagName=\"StatManager.AthenaKills\")" 0 false
        (:OnStatChanged
          (CallFunction|OnKill))))
    (:then_1
      (bind (_index _bindexisvalid) (Class|FortCosmeticFunctionLibrary|GetVariantChannelActiveIndexForItem "/BRCosmetics/Athena/Items/Cosmetics/Backpacks/Backpack_LastVoiceDive.Backpack_LastVoiceDive" (CosmeticLoadout|GetCosmeticLoadout) "(TagName=\"Cosmetics.Variant.Channel.Parts\")" (Utilities|Array|MakeArray "(TagName=\"Cosmetics.Variant.Property.Stage1\")") false false))
      (Variables|Default|SetreactivityEnabled _bindexisvalid))))

(event Custom|On%20Kill
  (if (Variables|Default|GetreactivityEnabled)
    (Audio|SpawnSoundAttached "/BRCosmetics/Sounds/Backblings/LastVoiceDive/Backbling_LastVoiceDive_React_Cue.Backbling_LastVoiceDive_React_Cue" (CustomCharacter|GetSkeletalMeshForPartOnPlayerPawn))))

(fn UpdateAudioStatValue
  (params (AudioComponent :type Array%20of%20Audio%20Component%20Object%20References) (InName :type Name :default "None") (StatValue :type Integer))
  (Audio|Parameter|SetIntegerParameter (Utilities|Array|Get(acopy) AudioComponent) InName StatValue :!class /Script/AudioExtensions.AudioParameterControllerInterface))

(fn ConstructionScript)
