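; Reactive cosmetic part modifier for QuarterClaspZoom that animates a character part's material as the player's kill count rises.
; In match it watches the AthenaKills stat; in the frontend a looping timer cycles a fake kill count for preview.
; Kills pushes the count and game-time timestamps into material parameters, staggering fade-in and three panners with delays.
; A kill sound then plays for locally viewed pawns outside the frontend.
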(event Custom|FrontEnd%20Cycle
  (switch int (Variables|Default|GetFrontEndInt)
    (:0
      _)
    (:1
      _)
    (:2
      _)
    (:3
      _))
  (bind _output_get (Variables|Default|SetFrontEndInt (+ (Variables|Default|GetFrontEndInt) 1)))
  (Variables|Default|SetNumberOfKills _output_get)
  (CallFunction|Kills)
  (if (>= (Variables|Default|GetFrontEndInt) 4)
    (bind _output_get_2 (Variables|Default|SetFrontEndInt 0))
    (Variables|Default|SetNumberOfKills _output_get_2)))

(event EventOnPostResIn
  (|CallParentFunction)
  (Rendering|Material|SetScalarParameterValueonMaterials (CustomCharacter|GetSkeletalMeshForPartOnPlayerPawn) "NumberOfKills" 1.0)
  (if (Reactive|Frontend|IsCosmeticObjectDisplayedinFrontendContext 0)
    (Rendering|Material|SetScalarParameterValueonMaterials (CustomCharacter|GetSkeletalMeshForPartOnPlayerPawn) "NumberOfKills" 1.0)
    (Utilities|Time|SetTimerbyEvent (EventDispatchers|CreateEvent :!event Custom|FrontEnd%20Cycle) 1e+01 true)
    (else
      (Rendering|Material|SetScalarParameterValueonMaterials (CustomCharacter|GetSkeletalMeshForPartOnPlayerPawn) "NumberOfKills" 0.0)
      (if (not (Variables|Default|GetIsKilled))
        (bind (_statname _statvalue) (Cosmetics|WatchCosmeticStat (CustomCharacter|GetAssociatedPlayerPawn) "(TagName=\"StatManager.AthenaKills\")" 0 false)
          (:OnStatChanged
            (bind _output_get (Variables|Default|SetNumberOfKills _statvalue))
            (if (> _output_get 0)
              (Variables|Default|SetIsKilled true)
              (CallFunction|Kills))))))))

(event Custom|Kills
  (Default|SetTimestepFadeIn)
  (Rendering|Material|SetScalarParameterValueonMaterials (CustomCharacter|GetSkeletalMeshForPartOnPlayerPawn) "NumberOfKills" (Math|Conversions|ToFloat(Integer) (Variables|Default|GetNumberOfKills)))
  ; start panning
  (Default|SetTimestepPanner)
  ; let start ramp up finish
  (Utilities|FlowControl|Delay (Variables|Default|GetTransitionTimeLengthStart))
  (Default|SetTimestepPannerB)
  ; let start ramp up finish
  (Utilities|FlowControl|Delay (Variables|Default|GetTransitionTimeLengthStart))
  (Default|SetTimestepPannerC)
  ; length of panner plus a delay before end ramp down
  (Utilities|FlowControl|Delay 0.4)
  ; play audio kill sound
  (CallFunction|AudioTrigger "(TagName=\"SoundLibrary.Cosmetic.Kills\")")
  (Utilities|Time|SetTimerbyEvent (EventDispatchers|CreateEvent :!event Custom|Reset%20to%20Default) (+ (Variables|Default|GetActivationLength) (Variables|Default|GetHoldLastFrameTime))))

(event Custom|Reset%20to%20Default
  (Variables|Default|SetIsKilled false))

(fn AudioTrigger
  (params (SoundLibTag :type Sound%20Library%20Tag%20Structure))
  (bind _returnvalue (CustomCharacter|GetAssociatedPlayerPawn))
  (if (and (Pawn|IsLocallyViewed _returnvalue) (not (Reactive|Frontend|IsCosmeticObjectDisplayedinFrontendContext 0)))
    (bind _returnvalue_2 (SoundLibrary|PlaySoundSimple (WorldSubsystems|GetSoundLibrarySubsystem) _returnvalue (Utilities|Struct|MakeSimpleSoundSettings SoundLibTag true false true false 1.0 1.0 0.0 0 0 "None" (CustomCharacter|GetSkeletalMeshForPartOnPlayerPawn)) :!class /Script/SoundLibrary.SoundLibrarySubsystem))
    (return (.bSuccess (Utilities|Struct|BreakSoundLibraryPlayResult _returnvalue_2)))))

(fn SetTimestepEnd
  (Rendering|Material|SetScalarParameterValueonMaterials (CustomCharacter|GetSkeletalMeshForPartOnPlayerPawn) "TimestampEnd" (Utilities|Time|GetGameTimeinSeconds))
  (Rendering|Material|SetScalarParameterValueonMaterials (CustomCharacter|GetSkeletalMeshForPartOnPlayerPawn) "TransitionLengthEnd" (Variables|Default|GetTransitionTimeLengthEnd))
  (return))

(fn SetTimestepFadeIn
  (Rendering|Material|SetScalarParameterValueonMaterials (CustomCharacter|GetSkeletalMeshForPartOnPlayerPawn) "TimestampFadeIn" (Utilities|Time|GetGameTimeinSeconds))
  (Rendering|Material|SetScalarParameterValueonMaterials (CustomCharacter|GetSkeletalMeshForPartOnPlayerPawn) "TransitionLengthFadeIn" (Variables|Default|GetTransitionTimeLengthFadeIn))
  (return))

(fn SetTimestepPanner
  (Rendering|Material|SetScalarParameterValueonMaterials (CustomCharacter|GetSkeletalMeshForPartOnPlayerPawn) "TimestampPanner" (Utilities|Time|GetGameTimeinSeconds))
  (Rendering|Material|SetScalarParameterValueonMaterials (CustomCharacter|GetSkeletalMeshForPartOnPlayerPawn) "TransitionLengthPanner" (Variables|Default|GetTransitionTimeLengthPanner))
  (return))

(fn SetTimestepPannerB
  (Rendering|Material|SetScalarParameterValueonMaterials (CustomCharacter|GetSkeletalMeshForPartOnPlayerPawn) "TimestampPannerB" (Utilities|Time|GetGameTimeinSeconds))
  (Rendering|Material|SetScalarParameterValueonMaterials (CustomCharacter|GetSkeletalMeshForPartOnPlayerPawn) "TransitionLengthPannerB" (Variables|Default|GetTransitionTimeLengthPannerB))
  (return))

(fn SetTimestepPannerC
  (Rendering|Material|SetScalarParameterValueonMaterials (CustomCharacter|GetSkeletalMeshForPartOnPlayerPawn) "TimestampPannerC" (Utilities|Time|GetGameTimeinSeconds))
  (Rendering|Material|SetScalarParameterValueonMaterials (CustomCharacter|GetSkeletalMeshForPartOnPlayerPawn) "TransitionLengthPannerC" (Variables|Default|GetTransitionTimeLengthPannerC))
  (return))

(fn SetTimestepStart
  (Rendering|Material|SetScalarParameterValueonMaterials (CustomCharacter|GetSkeletalMeshForPartOnPlayerPawn) "TimestampStart" (Utilities|Time|GetGameTimeinSeconds))
  (Rendering|Material|SetScalarParameterValueonMaterials (CustomCharacter|GetSkeletalMeshForPartOnPlayerPawn) "TransitionLengthStart" (Variables|Default|GetTransitionTimeLengthStart))
  (return))

(fn ConstructionScript
  (|CallParentFunction))
