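; Cosmetic part modifier for the GraffitiTon backpack that feeds player movement state into its idle VFX particle component.
; Tick normalizes pawn velocity, passes it to OnLinearVelocity and sets PlayerVelocityNormalized on the effect.
; After res-in, a looping timer polls crouching instead of checking per frame, using DoOnce gates around IsCrouched writes.
; Skydiving begin and end events toggle the IsSkydiving parameter.
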(event EventTick
  (params (DeltaSeconds :type Float%20%28single-precision%29))
  (|CallParentFunction DeltaSeconds)
  (bind (_normalizedoverallvelocity _normalizedwalkvelocity _velocitylength) (Velocity|GetNormalizedVelocity (Variables|Velocity|GetPawnVelocity)))
  (CallFunction|OnLinearVelocity _normalizedoverallvelocity _normalizedwalkvelocity _velocitylength)
  (Effects|Components|ParticleSystem|SetFloatParameter (CallFunction|GetIdleVFXComponent) "PlayerVelocityNormalized" _normalizedoverallvelocity))

(event EventOnPostResIn
  (|CallParentFunction)
  (Utilities|Time|SetTimerbyEvent (delegate Custom|CheckCrouch) 0.1 true))

(event Custom|CheckCrouch
  (bind _returnvalue (CallFunction|GetIdleVFXComponent))
  (Utilities|FlowControl|Sequence
    (:then_0
      (if (AI|Components|NavMovement|IsCrouching (Variables|/Script/Engine.Character|GetCharacterMovement (CustomCharacter|GetAssociatedPlayerPawn)))
        (Utilities|FlowControl|DoOnce false
          (:Completed
            (Utilities|FlowControl|Sequence
              (:then_0
                _)
              (:then_1
                (Utilities|FlowControl|DoOnce false
                  (:Completed
                    (Utilities|FlowControl|Sequence
                      (:then_0
                        _)
                      (:then_1
                        _))
                    (Effects|Components|ParticleSystem|SetBoolParameter _returnvalue "IsCrouched" false)))))
            (Effects|Components|ParticleSystem|SetBoolParameter _returnvalue "IsCrouched" true)))))))

(event EventOnBeginSkydiving
  (|CallParentFunction)
  (Effects|Components|ParticleSystem|SetBoolParameter (CallFunction|GetIdleVFXComponent) "IsSkydiving" true))

(event EventOnEndSkydiving
  (|CallParentFunction)
  (Effects|Components|ParticleSystem|SetBoolParameter (CallFunction|GetIdleVFXComponent) "IsSkydiving" false))

(fn ConstructionScript
  (|CallParentFunction))
