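; Cosmetic part modifier for the Nimble Pilot backpack that drives reactive material parameters from kills and movement.
; In matches, a watched kill stat sets Kill Counter and flags Double Digits from ten upward.
; Velocity updates feed Velocity and an interpolated Revolution parameter, switching actor tick interval between moving and still.
; In the frontend, a delayed timeline animates these parameters as a preview, then resets and replays through FE Reactivity.
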(event Custom|FE%20Reactivity
  (bind _returnvalue (CustomCharacter|GetAssociatedPlayerPawn))
  (bind _returnvalue_2 (Customization|GetSkeletalMeshforPartType _returnvalue "Backpack" :!class /Script/FortniteGame.FortPlayerPawn))
  (Rendering|Material|SetScalarParameterValueonMaterials _returnvalue_2 "Kill Counter" 0.0)
  (Utilities|FlowControl|Delay 5.0)
  (bind (_direction _kill_counter _revolution _velocity) (Timeline|Timeline PlayFromStart)
    (:Update
      (Rendering|Material|SetScalarParameterValueonMaterials _returnvalue_2 "Kill Counter" _kill_counter)
      (Rendering|Material|SetScalarParameterValueonMaterials _returnvalue_2 "Velocity" _velocity)
      (Rendering|Material|SetScalarParameterValueonMaterials _returnvalue_2 "Revolution" _revolution)
      (if (>= _kill_counter 1e+01)
        (Rendering|Material|SetScalarParameterValueonMaterials _returnvalue_2 "Double Digits" 1.0)))
    (:Finished
      (Rendering|Material|SetScalarParameterValueonMaterials _returnvalue_2 "Double Digits" 0.0)
      (CallFunction|FEReactivity))))

(event EventOnPostResIn
  (bind _returnvalue (CustomCharacter|GetAssociatedPlayerPawn))
  (bind _returnvalue_2 (Customization|GetSkeletalMeshforPartType _returnvalue "Backpack" :!class /Script/FortniteGame.FortPlayerPawn))
  (|CallParentFunction)
  (if (Reactive|Frontend|IsCosmeticObjectDisplayedinFrontendContext 0)
    (Rendering|Material|SetScalarParameterValueonMaterials _returnvalue_2 "Kill Counter" 0.0)
    (Utilities|FlowControl|Delay 5.0)
    (Timeline|Timeline PlayFromStart)
    (else
      (bind (_statname _statvalue) (Cosmetics|WatchCosmeticStat _returnvalue "(TagName=\"StatManager.AthenaKills\")" -1)
        (:OnStatChanged
          (bind _result (Math|Integer|IncrementInt (Variables|Default|GetKill%20Counter))
            (:%20%20
              (Variables|Default|SetKill%20Counter _result)
              (Rendering|Material|SetScalarParameterValueonMaterials _returnvalue_2 "Kill Counter" (Math|Conversions|ToFloat(Integer) _statvalue))
              (if (>= (Math|Conversions|ToFloat(Integer) _statvalue) 1e+01)
                (Rendering|Material|SetScalarParameterValueonMaterials _returnvalue_2 "Double Digits" 1.0)))))))))

(event EventOnDayPhaseChanged
  (params (CurrentDayPhase :type EFortDayPhase%20Enum :default "Morning") (PreviousDayPhase :type EFortDayPhase%20Enum :default "Morning") (bAtCreation :type Boolean)))

(event EventOnNewPlayerWeaponType
  (params (NewType :type EFortItemType%20Enum :default "WorldItem") (OldType :type EFortItemType%20Enum :default "WorldItem"))
  (|CallParentFunction NewType OldType) ; This node is disabled and will not be called.\nDrag off pins to build functionality.
)

(event EventOnWeaponChanged
  (params (NewWeapon :type Fort%20Weapon%20Object%20Reference) (OldWeapon :type Fort%20Weapon%20Object%20Reference))
  (|CallParentFunction NewWeapon OldWeapon) ; This node is disabled and will not be called.\nDrag off pins to build functionality.
)

(event EventOnWeaponFired)

(event EventOnNewWeapons
  (params (WeaponsAsItems :type Array%20of%20Fort%20Item%20Object%20References)))

(event EventOnCurieStateAttached
  (params (Tag :type Gameplay%20Tag%20Structure%20%28by%20ref%29)))

(event EventOnCurieStateDetached
  (params (Tag :type Gameplay%20Tag%20Structure%20%28by%20ref%29)))

(event EventOnBeginSkydiving
  (|CallParentFunction) ; This node is disabled and will not be called.\nDrag off pins to build functionality.
)

(event EventOnEndSkydiving
  (|CallParentFunction) ; This node is disabled and will not be called.\nDrag off pins to build functionality.
)

(event EventOnEnterVehicle
  (params (VehicleEntered :type Fort%20Vehicle%20Interface%20Interface%20%28by%20ref%29))
  (|CallParentFunction VehicleEntered) ; This node is disabled and will not be called.\nDrag off pins to build functionality.
)

(event EventOnExitVehicle
  (|CallParentFunction) ; This node is disabled and will not be called.\nDrag off pins to build functionality.
)

(event EventOnTargetingOrFirstPersonCameraChanged)

(event EventOnHitPawn
  (params (SelfActor :type Actor%20Object%20Reference) (OtherActor :type Actor%20Object%20Reference) (NormalImpulse :type Vector :default "0, 0, 0") (Hit :type Hit%20Result%20Structure%20%28by%20ref%29)))

(event EventOnLinearVelocity
  (params (NormalizedOverallVelocity :type Float%20%28double-precision%29) (NormalizedWalkVelocity :type Float%20%28double-precision%29) (VelocityLength :type Float%20%28double-precision%29))
  (bind _returnvalue (CustomCharacter|GetAssociatedPlayerPawn))
  (bind _returnvalue_2 (Customization|GetSkeletalMeshforPartType _returnvalue "Backpack" :!class /Script/FortniteGame.FortPlayerPawn))
  (bind _returnvalue_3 (Utilities|Time|GetWorldDeltaSeconds))
  (bind _revolution (Variables|Default|GetRevolution))
  (if (not (Reactive|Frontend|IsCosmeticObjectDisplayedinFrontendContext 0))
    (Rendering|Material|SetScalarParameterValueonMaterials _returnvalue_2 "Velocity" NormalizedOverallVelocity)
    (if (Math|Conversions|ToBoolean(Integer) (Math|Integer|Clamp(Integer) (Math|Float|Round (Math|Vector|VectorLength (Variables|Velocity|GetPawnVelocity))) 0 1))
      (Rendering|Material|SetScalarParameterValueonMaterials _returnvalue_2 "Revolution" (Math|Interpolation|FInterpTo _revolution 1.0 _returnvalue_3 0.5))
      (Actor|Tick|SetActorTickInterval 0.5)
      (Variables|Default|SetRevolution NormalizedOverallVelocity)
      (else
        (Actor|Tick|SetActorTickInterval 0.05)
        (bind _output_get (Variables|Default|SetRevolution (Math|Interpolation|FInterpTo _revolution 0.0 _returnvalue_3 0.5)))
        (Rendering|Material|SetScalarParameterValueonMaterials _returnvalue_2 "Revolution" _output_get)))))

(fn ConstructionScript
  (|CallParentFunction))
