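; Backpack part modifier for the SatinCheddar cosmetic that drives velocity-reactive material and idle VFX.
; In game, Tick feeds normalized accumulated pawn velocity, using zero while skydiving, to the material's Velocity and particle OwnerVelocity parameters.
; In the frontend, if ReactivityOn is set, a timeline simulates that velocity, replaying after a delay.
; Both paths deactivate the VFX while the "pack" bone is hidden, a noted bug fix.
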(event EventOnPostResIn
  (bind _returnvalue (CustomCharacter|GetSkeletalMeshForPartOnPlayerPawn))
  (bind _material_instance_dynamic (Class|BTechArtGlobalLibrary|GetDynamicMaterialInstance _returnvalue))
  (bind _returnvalue_2 (CustomCharacter|GetAssociatedPlayerPawn))
  (bind _returnvalue_3 (CustomCharacter|GetCharacterPartType))
  (bind _returnvalue_4 (Customization|GetSkeletalMeshforPartType _returnvalue_2 _returnvalue_3 :!class /Script/FortniteGame.FortPlayerPawn))
  (bind _returnvalue_5 (CallFunction|GetIdleVFXComponent))
  (|CallParentFunction)
  (bind _returnvalue_6 (Rendering|Material|GetScalarParameterValue _material_instance_dynamic "ReactivityOn" :!class /Script/Engine.MaterialInstanceDynamic))
  (if (and (>= _returnvalue_6 1.0) (Reactive|Frontend|IsCosmeticObjectDisplayedinFrontendContext 0))
    (bind (_direction _feloopcurve) (Timeline|FrontendTimeline PlayFromStart)
      (:Update
        (Rendering|Material|SetScalarParameterValue _material_instance_dynamic "Velocity" _feloopcurve :!class /Script/Engine.MaterialInstanceDynamic)
        (Effects|Components|ParticleSystem|SetFloatParameter (CallFunction|GetIdleVFXComponent) "OwnerVelocity" (* _feloopcurve 2.0))
        (Utilities|IsValid _returnvalue_4
          (:Is%20Valid
            (bind _returnvalue_7 (Components|SkinnedMesh|IsBoneHiddenbyName _returnvalue_4 "pack"))
            ;[0.231 0.5365 0.19 0.5] Created in 08/20 to fix the Jira FORT-780574
            (if _returnvalue_7
              (Components|Activation|Deactivate _returnvalue_5)
              (else
                (Components|Activation|Activate _returnvalue_5 false))))))
      (:Finished
        (Utilities|FlowControl|Delay 3.0)
        (Timeline|FrontendTimeline PlayFromStart)))))

(event EventTick
  (params (DeltaSeconds :type Float%20%28single-precision%29))
  (bind _returnvalue (CustomCharacter|GetAssociatedPlayerPawn))
  (bind _returnvalue_2 (CustomCharacter|GetCharacterPartType))
  (bind _returnvalue_3 (Customization|GetSkeletalMeshforPartType _returnvalue _returnvalue_2 :!class /Script/FortniteGame.FortPlayerPawn))
  (bind _returnvalue_4 (CallFunction|GetIdleVFXComponent))
  (|CallParentFunction DeltaSeconds)
  (if (not (Reactive|Frontend|IsCosmeticObjectDisplayedinFrontendContext 0))
    (bind (_normalized_accumulatedvelocity _accumulated_velocity) (Velocity|GetNormalizedAccumulatedVelocity DeltaSeconds (stale _output_get) 7e+04 1.0 1.25 0.1 (Math|Vector|SelectVector "0, 0, 0" (Variables|Velocity|GetPawnVelocity) (Parachute|IsSkydiving (CustomCharacter|GetAssociatedPlayerPawn)))))
    (bind _output_get (Variables|Velocity|SetAccumulated%20Velocity _accumulated_velocity))
    (Rendering|Material|SetScalarParameterValue (Class|BTechArtGlobalLibrary|GetDynamicMaterialInstance (CustomCharacter|GetSkeletalMeshForPartOnPlayerPawn)) "Velocity" _normalized_accumulatedvelocity :!class /Script/Engine.MaterialInstanceDynamic)
    (Effects|Components|ParticleSystem|SetFloatParameter (CallFunction|GetIdleVFXComponent) "OwnerVelocity" _normalized_accumulatedvelocity)
    (Utilities|IsValid _returnvalue_3
      (:Is%20Valid
        (bind _returnvalue_5 (Components|SkinnedMesh|IsBoneHiddenbyName _returnvalue_3 "pack"))
        ;[0.231 0.5365 0.19 0.5] Created in 08/20 to fix the Jira FORT-780574
        (if _returnvalue_5
          (Components|Activation|Deactivate _returnvalue_4)
          (else
            (Components|Activation|Activate _returnvalue_4 false)))))))

(event EventOnTacticalSprintingToggle
  (params (Is%20TacticalSprinting :type Boolean)))

(fn ConstructionScript
  (|CallParentFunction))
