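; B_PartModifier_CustomDepth_Toon is a cosmetic character part modifier Blueprint for toon-style outfits that turns on custom depth rendering.
; On the post respawn-in event (OnPostResIn) it calls the parent, waits a short delay, then gets the owner's skeletal mesh for the part and enables Render CustomDepth on it.
; The construction script only calls its parent.
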(event EventOnPostResIn
  (|CallParentFunction)
  (Utilities|FlowControl|Delay 0.15)
  (Rendering|SetRenderCustomDepth (CustomCharacter|GetSkeletalMeshforPartonOwner) true))

(fn ConstructionScript
  (|CallParentFunction))
