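; Cosmetic part modifier for DryEraseCod that drives audio-reactive material fades on the character part's skeletal mesh.
; In frontend or preview, a looping timeline with on/off delays fakes audio amplitude, feeding the mesh materials and idle VFX.
; Otherwise, a looping timer polls Amplitude_Music from the audio-reactive parameter collection.
; That poll triggers DoOnce-gated timelines for FadingIn, and for FadingOut when amplitude drops to zero after music has played.
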(event Custom|OnFakeLoopTimelineUpdate
  (params (TransitionValue :type Float%20%28single-precision%29))
  (Rendering|Material|SetScalarParameterValueonMaterials (CustomCharacter|GetSkeletalMeshForPartOnPlayerPawn) "FadingIn" 1.0)
  (Rendering|Material|SetScalarParameterValueonMaterials (CustomCharacter|GetSkeletalMeshForPartOnPlayerPawn) "EnableFrontendFakeAudioLoop" TransitionValue) ; Enable Fake Audio in Frontend for Materials
  (Effects|Components|ParticleSystem|SetFloatParameter (CallFunction|GetIdleVFXComponent) "EnableFrontendFakeAudioLoop" TransitionValue))

(event EventOnPostResIn
  (|CallParentFunction)
  (if (or (Reactive|Frontend|IsCosmeticObjectDisplayedinFrontendContext 0) (Reactive|Frontend|IsCosmeticObjectDisplayedasPreview 0))
    (Variables|Default|SetBeginLoop true :!label _label)
    (bind (_direction _in) (Timeline|InOutTimeline PlayFromStart)
      (:Update
        (CallFunction|OnFakeLoopTimelineUpdate _in))
      (:Finished
        (if (Variables|Default|GetBeginLoop)
          (Variables|Default|SetBeginLoop false)
          (Utilities|FlowControl|Delay (Variables|Default|GetFakeAmplitudeOnLength))
          (Timeline|InOutTimeline ReverseFromEnd)
          (else
            (Utilities|FlowControl|Delay (Variables|Default|GetFakeAmplitudeOffLength))
            (goto _label)))))
    (else
      (Utilities|Time|SetTimerbyEvent (delegate Custom|PlayingMusic) 0.05 true false 0.5))))

(event Custom|PlayingMusic
  (bind _returnvalue (Rendering|Material|GetScalarParameterValue "/Game/Athena/Events/AudioAnalysis/MPC_Audio_Reactive_Parameters.MPC_Audio_Reactive_Parameters" "Amplitude_Music" :!class /Script/Engine.KismetMaterialLibrary))
  (if (> _returnvalue 0)
    (Variables|Default|SetPlayed true)
    (elif (Variables|Default|GetPlayed)
      (Utilities|FlowControl|DoOnce false :!label _label
        (:Completed
          (bind (_direction _fading_out) (Timeline|Timeline_0 PlayFromStart)
            (:Update
              (Rendering|Material|SetScalarParameterValueonMaterials (CustomCharacter|GetSkeletalMeshForPartOnPlayerPawn) "FadingOut" _fading_out))
            (:Finished
              (Variables|Default|SetPlayed false)
              (Rendering|Material|SetScalarParameterValueonMaterials (CustomCharacter|GetSkeletalMeshForPartOnPlayerPawn) "FadingIn" 0.0)))))
      (else
        (goto _label Reset))))
  (Utilities|FlowControl|DoOnce false
    (:Completed
      (bind (_direction_2 _fading_in) (Timeline|Timeline PlayFromStart)
        (:Update
          (Rendering|Material|SetScalarParameterValueonMaterials (CustomCharacter|GetSkeletalMeshForPartOnPlayerPawn) "FadingIn" _fading_in))))))

(fn ConstructionScript
  (|CallParentFunction))
