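; Cosmetic part modifier for the StridemiceDeep audio-reactive character part, faking audio reactivity in frontend and preview displays.
; On post res-in, an in/out timeline fades "EnableFrontendFakeAudioLoop" on mesh materials and idle VFX, using delays and reversal to hold, then fade out.
; A separate LED transition event animates "LED Transition" and "LedTime2" material scalars through a timeline, zeroing the former when finished.
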(event Custom|OnFakeLoopTimelineUpdate
  (params (TransitionValue :type Float%20%28single-precision%29))
  (Rendering|Material|SetScalarParameterValueonMaterials (CustomCharacter|GetSkeletalMeshForPartOnPlayerPawn) "EnableFrontendFakeAudioLoop" TransitionValue) ; Enable Fake Audio in Frontend for Materials
  (Effects|Components|ParticleSystem|SetFloatParameter (CallFunction|GetIdleVFXComponent) "EnableFrontendFakeAudioLoop" TransitionValue))

(event EventOnPostResIn
  (|CallParentFunction)
  (if (or (Reactive|Frontend|IsCosmeticObjectDisplayedinFrontendContext 0) (Reactive|Frontend|IsCosmeticObjectDisplayedasPreview 0))
    (Variables|Default|SetBeginLoop true)
    (bind (_direction _in) (Timeline|InOutTimeline PlayFromStart)
      (:Update
        (CallFunction|OnFakeLoopTimelineUpdate _in))
      (:Finished
        (if (Variables|Default|GetBeginLoop)
          (Variables|Default|SetBeginLoop false)
          (Utilities|FlowControl|Delay (Variables|Default|GetFakeAmplitudeOnLength))
          (Timeline|InOutTimeline ReverseFromEnd)
          (else
            (Utilities|FlowControl|Delay (Variables|Default|GetFakeAmplitudeOffLength))))))))

(event Custom|Play%20LED%20transition
  (bind _returnvalue (CustomCharacter|GetSkeletalMeshForPartOnPlayerPawn))
  ;[1 0 0.1124 0.5] Useless
  (bind (_direction _led_transition _led) (Timeline|Timeline PlayFromStart)
    (:Update
      (Rendering|Material|SetScalarParameterValueonMaterials _returnvalue "LED Transition" _led_transition) ; Enable Fake Audio in Frontend for Materials
      (Rendering|Material|SetScalarParameterValueonMaterials (CustomCharacter|GetSkeletalMeshForPartOnPlayerPawn) "LedTime2" _led) ; Enable Fake Audio in Frontend for Materials
    )
    (:Finished
      (Rendering|Material|SetScalarParameterValueonMaterials _returnvalue "LED Transition" 0.0) ; Enable Fake Audio in Frontend for Materials
    )))

(fn ConstructionScript
  (|CallParentFunction))
