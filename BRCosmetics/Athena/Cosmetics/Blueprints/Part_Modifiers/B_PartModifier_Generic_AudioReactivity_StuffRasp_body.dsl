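; Body part modifier for the StuffRasp cosmetic that makes its materials audio-reactive and sets custom depth and translucency sorting on the pawn.
; In frontend or preview, delayed timelines drive fake-audio parameters on materials and idle VFX.
; Otherwise, a 0.1-second timer checks cubed music amplitude from a parameter collection and plays a guarded reactive-layer animation above a threshold.
; Handlers marked Outdate set glove-reactive overlay parameters on pickaxe and body meshes.
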(event Custom|MaterialAutoSwitch
  ;[1 0 0 0.5] Outdate
  (bind _asfort_pawn (Utilities|Casting|CastToFortPawn (Game|GetPlayerCharacter 0)))
  (bind _asfort_weapon_pickaxe_athena (Utilities|Casting|CastToFortWeaponPickaxeAthena (Variables|/Script/FortniteGame.FortPawn|GetCurrentWeapon _asfort_pawn)))
  (for _array_element (Visual|GetWeaponMeshes _asfort_weapon_pickaxe_athena)
    (Class|FortCosmeticMaterialBlueprintLibrary|SetScalarParameterValueonOverlayMaterial _array_element "StuffRasp_gloveReactive" 1.0)))

(event Custom|CustomEvent
  ;[1 0 0 0.5] Outdate
  (CallFunction|MaterialAutoSwitch))

(event EventOnPawnSet
  ;[1 0 0 0.5] Outdate
  (CallFunction|MaterialAutoSwitch))

(event Custom|CustomEvent_0
  ;[1 0 0 0.5] Outdate
  (Class|FortCosmeticMaterialBlueprintLibrary|SetScalarParameterValueonOverlayMaterial (CustomCharacter|GetSkeletalMeshForPartOnPlayerPawn) "StuffRasp_gloveReactive" 1.0))

(event EventOnPawnUpdated
  ;[1 0 0 0.5] Outdate
  (CallFunction|MaterialAutoSwitch))

(event Custom|Fakeaudio
  (params (Transition%20Value :type Float%20%28double-precision%29))
  ; Fake Audio Reactivity for Previews and Front End Context
  (Rendering|Material|SetScalarParameterValueonMaterials (CustomCharacter|GetSkeletalMeshForPartOnPlayerPawn) "EnableFrontendFakeAudioLoop" Transition%20Value) ; Enable Fake Audio in Frontend for Materials
  (Effects|Components|ParticleSystem|SetFloatParameter (CallFunction|GetIdleVFXComponent) "EnableFrontendFakeAudioLoop" Transition%20Value)
  (bind (_direction _in) (Timeline|Timeline_1 PlayFromStart)
    (:Update
      (Rendering|Material|SetScalarParameterValueonMaterials (CustomCharacter|GetSkeletalMeshForPartOnPlayerPawn) "FakeAudioTime2" _in) ; Enable Fake Audio in Frontend for Materials
    )))

(event EventOnPostResIn
  (bind _returnvalue (CustomCharacter|GetAssociatedPlayerPawn))
  (bind _returnvalue_2 (Customization|GetSkeletalMeshforPartType _returnvalue "Body" :!class /Script/FortniteGame.FortPlayerPawn))
  (bind _returnvalue_3 (CustomCharacter|GetAssociatedPlayerPawn))
  (bind _returnvalue_4 (Customization|GetSkeletalMeshforPartType _returnvalue_3 "Body" :!class /Script/FortniteGame.FortPlayerPawn))
  (|CallParentFunction)
  (Rendering|SetRenderCustomDepth _returnvalue_4 true)
  (Rendering|SetTranslucentSortPriority _returnvalue_4 3)
  (Rendering|SetTranslucentSortPriority (Customization|GetSkeletalMeshforPartType (CustomCharacter|GetAssociatedPlayerPawn) "Face" :!class /Script/FortniteGame.FortPlayerPawn) 1)
  (if (or (Reactive|Frontend|IsCosmeticObjectDisplayedinFrontendContext 0) (Reactive|Frontend|IsCosmeticObjectDisplayedasPreview 0))
    (Utilities|FlowControl|Sequence
      (:then_0
        ; Fake Audio Reactivity for Previews and Front End Context
        (Utilities|FlowControl|Delay (Variables|AudioReactivity|GetAudioReactivityFakeAmplitudeOffLength))
        (bind (_direction _in) (Timeline|Timeline_0 Play)
          (:Update
            (CallFunction|Fakeaudio _in))
          (:Finished
            (switch Utilities|FlowControl|Switch|SwitchonETimelineDirection _direction
              (:Forward
                (Utilities|FlowControl|Delay (Variables|AudioReactivity|GetAudioReactivityFakeAmplitudeOnLength))
                (Timeline|Timeline_0 ReverseFromEnd))
              (:Backward)))))
      (:then_1
        (Utilities|FlowControl|Delay 2.0)
        (Rendering|SetRenderCustomDepth _returnvalue_2 true)
        (Rendering|SetTranslucentSortPriority _returnvalue_2 3)))
    (else
      (Utilities|Time|SetTimerbyEvent (delegate Custom|PlayingReactive) 0.1 true))))

(event Custom|PlayingReactive
  (bind _returnvalue (Rendering|Material|GetScalarParameterValue "/Game/Athena/Events/AudioAnalysis/MPC_Audio_Reactive_Parameters.MPC_Audio_Reactive_Parameters" "Amplitude_Music" :!class /Script/Engine.KismetMaterialLibrary))
  (if (Variables|Default|GetPlayed)
    _
    (elif (> (Math|Float|Power _returnvalue 3.0) 0.6)
      (Variables|Default|SetPlayed true)
      (bind (_direction _fading_in) (Timeline|Timeline PlayFromStart)
        (:Update
          (Rendering|Material|SetScalarParameterValueonMaterials (CustomCharacter|GetSkeletalMeshForPartOnPlayerPawn) "ReactiveAnim_Layer2" _fading_in))
        (:Finished
          (Variables|Default|SetPlayed false))))))

(fn ConstructionScript
  (|CallParentFunction))
