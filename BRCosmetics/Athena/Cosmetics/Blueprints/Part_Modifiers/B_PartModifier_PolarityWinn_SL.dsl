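; Character part modifier for the PolarityWinn cosmetic that drives material and idle VFX parameters.
; In frontend or preview, a timeline fakes audio reactivity by ramping EnableFrontendFakeAudioLoop up, holding, then reversing it.
; On res-in it sets BP_UseCCPM and enables tick unless the Emissive variant index is zero.
; Tick feeds head-to-neck offset and upper-spine socket position into the mesh materials as chest parameters.
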(event Custom|OnFakeLoopTimelineUpdate
  (params (TransitionValue :type Float%20%28single-precision%29))
  (Rendering|Material|SetScalarParameterValueonMaterials (CustomCharacter|GetSkeletalMeshForPartOnPlayerPawn) "EnableFrontendFakeAudioLoop" TransitionValue) ; Enable Fake Audio in Frontend for Materials
  (Effects|Components|ParticleSystem|SetFloatParameter (CallFunction|GetIdleVFXComponent) "EnableFrontendFakeAudioLoop" TransitionValue))

(event EventOnPostResIn
  (|CallParentFunction)
  (Utilities|FlowControl|Sequence
    (:then_0
      (if (or (Reactive|Frontend|IsCosmeticObjectDisplayedinFrontendContext 0) (Reactive|Frontend|IsCosmeticObjectDisplayedasPreview 0))
        (Variables|Default|SetBeginLoop true)
        (bind (_direction _in) (Timeline|InOutTimeline PlayFromStart)
          (:Update
            (CallFunction|OnFakeLoopTimelineUpdate _in))
          (:Finished
            (if (Variables|Default|GetBeginLoop)
              (Variables|Default|SetBeginLoop false)
              (Utilities|FlowControl|Delay (Variables|Default|GetFakeAmplitudeOnLength))
              (Timeline|InOutTimeline ReverseFromEnd)
              (else
                (Utilities|FlowControl|Delay (Variables|Default|GetFakeAmplitudeOffLength))))))))
    (:then_1
      (Rendering|Material|SetScalarParameterValueonMaterials (CustomCharacter|GetSkeletalMeshForPartOnPlayerPawn) "BP_UseCCPM" 1.0)
      (bind _index (Class|FortCosmeticFunctionLibrary|GetVariantChannelActiveIndexForSlot "(TagName=\"LoadoutSlot.Character\")" (CosmeticLoadout|GetCosmeticLoadout) "(TagName=\"Cosmetics.Variant.Channel.Emissive\")" (Utilities|Array|MakeArray :%5B0%5D "(TagName=\"Cosmetics.Variant.Property.Emissive1\")" :%5B1%5D "(TagName=\"Cosmetics.Variant.Property.Emissive2\")" :%5B2%5D "(TagName=\"Cosmetics.Variant.Property.Emissive3\")" :%5B3%5D "(TagName=\"Cosmetics.Variant.Property.Emissive4\")")))
      (if (== _index 0)
        (Actor|Tick|SetActorTickEnabled false)
        (else
          (Actor|Tick|SetActorTickEnabled true))))))

(event EventTick
  (params (DeltaSeconds :type Float%20%28single-precision%29))
  (bind _returnvalue (CustomCharacter|GetSkeletalMeshForPartOnPlayerPawn))
  (Rendering|Material|SetVectorParameterValueonMaterials _returnvalue "BP_OffsetVector_Chest" (- (Transformation|GetSocketLocation _returnvalue "head") (Transformation|GetSocketLocation _returnvalue "neck_02")))
  (Rendering|Material|SetVectorParameterValueonMaterials _returnvalue "BP_ChestPosition" (Transformation|GetSocketLocation _returnvalue "spine_05")))

(fn ConstructionScript
  (|CallParentFunction))
