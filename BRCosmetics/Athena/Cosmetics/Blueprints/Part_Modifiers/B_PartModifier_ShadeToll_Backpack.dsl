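; Cosmetic part modifier for the ShadeToll backpack that drives a heart-rate material effect from the pawn's activity.
; Velocity updates interpolate HeartRateState up while skydiving or moving and slowly down at rest, writing it to the backpack mesh materials.
; A watched DamageReceived stat also raises it, using game time as the interpolation delta.
; The HighBMPBool parameter switches on immediately for damage or skydiving, above a threshold when moving, and off after a delay when calm.
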(event EventOnPostResIn
  (bind _returnvalue (Customization|GetSkeletalMeshforPartType (CustomCharacter|GetAssociatedPlayerPawn) "Backpack" :!class /Script/FortniteGame.FortPlayerPawn))
  (bind _returnvalue_2 (CustomCharacter|GetAssociatedPlayerPawn))
  (|CallParentFunction)
  (bind (_statname _statvalue) (Cosmetics|WatchCosmeticStat _returnvalue_2 "(TagName=\"StatManager.DamageReceived\")")
    (:OnStatChanged
      (if (> _statvalue 0)
        (bind _output_get (Variables|Default|SetHeartRateState (Math|Interpolation|FInterpTo (Variables|Default|GetHeartRateState) 1.0 (Utilities|Time|GetGameTimeinSeconds) 0.95)))
        (Rendering|Material|SetScalarParameterValueonMaterials _returnvalue "HeartRateState" _output_get)
        (bind _output_get_2 (Variables|Default|SetHighBPM 1.0))
        (Rendering|Material|SetScalarParameterValueonMaterials _returnvalue "HighBMPBool" _output_get_2)))))

(event EventOnPawnAndPartChanged
  (|CallParentFunction))

(event EventOnVelocityToggle
  (params (IsMoving :type Boolean) (Delta%20Seconds :type Float%20%28double-precision%29))
  (bind _returnvalue (Customization|GetSkeletalMeshforPartType (CustomCharacter|GetAssociatedPlayerPawn) "Backpack" :!class /Script/FortniteGame.FortPlayerPawn))
  (bind _returnvalue_2 (CustomCharacter|GetAssociatedPlayerPawn))
  (Utilities|IsSkydiving _returnvalue_2
    (:true
      (bind _output_get (Variables|Default|SetHeartRateState (Math|Interpolation|FInterpTo (Variables|Default|GetHeartRateState) 1.0 Delta%20Seconds 0.9)))
      (Rendering|Material|SetScalarParameterValueonMaterials _returnvalue "HeartRateState" _output_get)
      (bind _output_get_2 (Variables|Default|SetHighBPM 1.0))
      (Rendering|Material|SetScalarParameterValueonMaterials _returnvalue "HighBMPBool" _output_get_2))
    (:false
      (if IsMoving
        (bind _output_get_3 (Variables|Default|SetHeartRateState (Math|Interpolation|FInterpTo (Variables|Default|GetHeartRateState) 1.0 Delta%20Seconds 0.5)))
        (Rendering|Material|SetScalarParameterValueonMaterials _returnvalue "HeartRateState" _output_get_3)
        (if (>= (Variables|Default|GetHeartRateState) 0.7)
          (bind _output_get_4 (Variables|Default|SetHighBPM 1.0))
          (Rendering|Material|SetScalarParameterValueonMaterials _returnvalue "HighBMPBool" _output_get_4))
        (else
          (bind _output_get_5 (Variables|Default|SetHeartRateState (Math|Interpolation|FInterpTo (Variables|Default|GetHeartRateState) 0.0 Delta%20Seconds 0.08)))
          (Rendering|Material|SetScalarParameterValueonMaterials _returnvalue "HeartRateState" _output_get_5)
          (if (<= (Variables|Default|GetHeartRateState) 0.5)
            (bind _output_get_6 (Variables|Default|SetHighBPM 0.0))
            (Utilities|FlowControl|Delay 4.0)
            (Rendering|Material|SetScalarParameterValueonMaterials _returnvalue "HighBMPBool" _output_get_6)))))))

(fn ConstructionScript
  (|CallParentFunction))
