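; Character part modifier for the Majesty backpack that drives its "SuperLevelling" material look from consolidated quest progress.
; Quest updates and post-res-in push level range, radius and highest count to material scalars; quest updates also refresh the Corruption slider variant.
; A looping workaround timer re-applies parameters until a delayed clear.
; Counts inside the level range trigger a gated change event that stamps game time and spawns Niagara FX once.
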(event Custom|HackyWorkaroundRefreshTimerGuy
  ;[1 1 1] SuperLevelling Logic
  (Utilities|FlowControl|Sequence
    (:then_0
      (CallFunction|InitializeVariables)
      (Rendering|Material|SetScalarParameterValueonMaterials (CustomCharacter|GetSkeletalMeshForPartOnPlayerPawn) "SL_CurrentLevel" (Math|Conversions|ToFloat(Integer) (Variables|Default|GetSL_LastValue)))
      (Utilities|FlowControl|DoOnce false
        (:Completed
          (bind _returnvalue (Utilities|Time|SetTimerbyEvent (EventDispatchers|CreateEvent :!event Custom|HackyWorkaroundRefreshTimerGuy) 0.1 true))
          (Variables|Default|SetSL_RefreshTimeHandler _returnvalue))))
    (:then_1
      (Utilities|FlowControl|Gate true
        (:Exit
          (if (and (> (Math|Integer|Max(Integer) (stale OnCustomUpdateObjective(MajestyConsolidatedQuest) AchievedCount) (Variables|Default|GetSL_LastValue)) (Variables|Default|GetSL_StartLevel)) (< (Math|Integer|Max(Integer) (stale OnCustomUpdateObjective(MajestyConsolidatedQuest) AchievedCount) (Variables|Default|GetSL_LastValue)) (Variables|Default|GetSL_EndLevel)))
            (CallFunction|OnActualStatChange)))))))

(event Custom|ResetSafetyGate
  ;[1 1 1] SuperLevelling Logic
  (Utilities|FlowControl|DoOnce false
    (:Completed
      (bind _returnvalue (Utilities|Time|SetTimerbyEvent (EventDispatchers|CreateEvent :!event Custom|HackyWorkaroundRefreshTimerGuy) 0.1 true))
      (Variables|Default|SetSL_RefreshTimeHandler _returnvalue))))

(event EventOnPostResIn
  (bind _returnvalue (CustomCharacter|GetSkeletalMeshForPartOnPlayerPawn))
  (bind _sl_lastvalue (Variables|Default|GetSL_LastValue))
  (|CallParentFunction)
  (CallFunction|ResetSafetyGate)
  ;[1 1 1] SuperLevelling Logic
  (Utilities|FlowControl|Sequence
    (:then_0
      (Utilities|FlowControl|Delay 1.0)
      (Utilities|Time|ClearandInvalidateTimerbyHandle (Variables|Default|GetSL_RefreshTimeHandler)))
    (:then_1
      (Rendering|Material|SetScalarParameterValueonMaterials _returnvalue "SL_StartAtLevel" (Math|Conversions|ToFloat(Integer) (Variables|Default|GetSL_StartLevel)))
      (Rendering|Material|SetScalarParameterValueonMaterials _returnvalue "SL_StopAtLevel" (Math|Conversions|ToFloat(Integer) (Variables|Default|GetSL_EndLevel)))
      (Rendering|Material|SetScalarParameterValueonMaterials _returnvalue "SL_EndRadius" (Variables|Default|GetEndRadius))
      (Variables|Default|SetSL_LastValue (Math|Integer|Max(Integer) (stale OnCustomUpdateObjective(MajestyConsolidatedQuest) AchievedCount) _sl_lastvalue))
      (Utilities|FlowControl|Sequence
        (:then_0
          (CallFunction|InitializeVariables)
          (Rendering|Material|SetScalarParameterValueonMaterials (CustomCharacter|GetSkeletalMeshForPartOnPlayerPawn) "SL_CurrentLevel" (Math|Conversions|ToFloat(Integer) (Variables|Default|GetSL_LastValue)))
          (Utilities|FlowControl|DoOnce false
            (:Completed
              (bind _returnvalue_2 (Utilities|Time|SetTimerbyEvent (EventDispatchers|CreateEvent :!event Custom|HackyWorkaroundRefreshTimerGuy) 0.1 true))
              (Variables|Default|SetSL_RefreshTimeHandler _returnvalue_2))))
        (:then_1
          (Utilities|FlowControl|Gate true
            (:Exit
              (if (and (> (Math|Integer|Max(Integer) (stale OnCustomUpdateObjective(MajestyConsolidatedQuest) AchievedCount) _sl_lastvalue) (Variables|Default|GetSL_StartLevel)) (< (Math|Integer|Max(Integer) (stale OnCustomUpdateObjective(MajestyConsolidatedQuest) AchievedCount) _sl_lastvalue) (Variables|Default|GetSL_EndLevel)))
                (CallFunction|OnActualStatChange)))))))))

(event Custom|OnActualStatChange
  (bind _returnvalue (CustomCharacter|GetSkeletalMeshForPartOnPlayerPawn))
  ;[1 1 1] SuperLevelling Logic
  ;;[1 1 1] This should only fire off when the stat changes - use this event!
  (Rendering|Material|SetScalarParameterValueonMaterials _returnvalue "SL_GameTimeFromBP" (Utilities|Time|GetGameTimeinSeconds))
  (Utilities|FlowControl|DoOnce false
    (:Completed
      (Effects|SpawnFXSystemAttached "/BRCosmetics/Characters/Player/Male/Medium/Bodies/M_MED_Majesty/FX/NS_Majesty_Backpack_Trigger.NS_Majesty_Backpack_Trigger" _returnvalue "pack")
      (Utilities|FlowControl|RetriggerableDelay 0.2))))

(event OnCustomUpdateObjective(MajestyConsolidatedQuest)
  (params (PlayerController :type Fort%20Player%20Controller%20Object%20Reference) (AchievedCount :type Integer) (RequiredCount :type Integer))
  (bind _returnvalue (CustomCharacter|GetSkeletalMeshForPartOnPlayerPawn))
  (bind _sl_lastvalue (Variables|Default|GetSL_LastValue))
  (bind _returnvalue_2 (Math|Integer|Max(Integer) AchievedCount _sl_lastvalue))
  (Utilities|FlowControl|Sequence
    (:then_0
      ;[1 1 1] SuperLevelling Logic
      (Utilities|FlowControl|Sequence
        (:then_0
          (Utilities|FlowControl|Delay 1.0)
          (Utilities|Time|ClearandInvalidateTimerbyHandle (Variables|Default|GetSL_RefreshTimeHandler)))
        (:then_1
          (Rendering|Material|SetScalarParameterValueonMaterials _returnvalue "SL_StartAtLevel" (Math|Conversions|ToFloat(Integer) (Variables|Default|GetSL_StartLevel)))
          (Rendering|Material|SetScalarParameterValueonMaterials _returnvalue "SL_StopAtLevel" (Math|Conversions|ToFloat(Integer) (Variables|Default|GetSL_EndLevel)))
          (Rendering|Material|SetScalarParameterValueonMaterials _returnvalue "SL_EndRadius" (Variables|Default|GetEndRadius))
          (Variables|Default|SetSL_LastValue _returnvalue_2)
          (Utilities|FlowControl|Sequence
            (:then_0
              (CallFunction|InitializeVariables)
              (Rendering|Material|SetScalarParameterValueonMaterials (CustomCharacter|GetSkeletalMeshForPartOnPlayerPawn) "SL_CurrentLevel" (Math|Conversions|ToFloat(Integer) (Variables|Default|GetSL_LastValue)))
              (Utilities|FlowControl|DoOnce false
                (:Completed
                  (bind _returnvalue_3 (Utilities|Time|SetTimerbyEvent (EventDispatchers|CreateEvent :!event Custom|HackyWorkaroundRefreshTimerGuy) 0.1 true))
                  (Variables|Default|SetSL_RefreshTimeHandler _returnvalue_3))))
            (:then_1
              (Utilities|FlowControl|Gate true
                (:Exit
                  (if (and (> _returnvalue_2 (Variables|Default|GetSL_StartLevel)) (< _returnvalue_2 (Variables|Default|GetSL_EndLevel)))
                    (CallFunction|OnActualStatChange)))))))))
    (:then_1
      (CustomCharacter|UpdateLockedFloatSliderVariant (Variables|Default|GetCosmeticItemDefinition) PlayerController "(TagName=\"Cosmetics.Variant.Channel.Corruption\")"))))

(fn GetAnimInstance
  (return (Components|SkeletalMesh|GetAnimInstance (Customization|GetSkeletalMeshforPartType (CustomCharacter|GetAssociatedPlayerPawn) (CustomCharacter|GetCharacterPartType) :!class /Script/FortniteGame.FortPlayerPawn))))

(fn InitializeVariables
  (bind _returnvalue (CustomCharacter|GetSkeletalMeshForPartOnPlayerPawn))
  (Rendering|Material|SetScalarParameterValueonMaterials _returnvalue "SL_StartAtLevel" (Math|Conversions|ToFloat(Integer) (Variables|Default|GetSL_StartLevel)))
  (Rendering|Material|SetScalarParameterValueonMaterials _returnvalue "SL_StopAtLevel" (Math|Conversions|ToFloat(Integer) (Variables|Default|GetSL_EndLevel)))
  (Rendering|Material|SetScalarParameterValueonMaterials _returnvalue "SL_EndRadius" (Variables|Default|GetEndRadius)))

(fn ConstructionScript
  (|CallParentFunction))
